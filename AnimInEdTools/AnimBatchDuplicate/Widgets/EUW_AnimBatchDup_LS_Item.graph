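# Editor utility list row for one Level Sequence in the animation batch duplicate tool, showing its name and a Source-to-Target renamed preview.
# UpdateWidget fills nested lists with the sequence's linked AnimSequences and, via Asset Registry referencers, the AnimMontages referencing them, by name without loading them.
# SetSelected writes the checkbox state to the list item data and cascades it to child rows.
# UpdateTargets pushes replacement strings down to those rows.
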EditorUtilityWidgetBlueprint /AnimInEdTools/AnimBatchDuplicate/Widgets/EUW_AnimBatchDup_LS_Item
graphs [
 EventGraph {nodes [
   n1 = "On Entry Initialized (AnimSequencesListView)"!() -> (OutputDelegate, then, Item:Object Reference, Widget:User Widget Object Reference) {is=title K2Node_ComponentBoundEvent}
   n2 = Switch on ECheckBoxState!(
    execute=[@n1.then, @n4.then, @n5.then, @n6.then]
    Selection:ECheckBoxState Enum=GetCheckedState!(self:Check Box Object Reference=Get LevelSequenceCheckbox() -> (LevelSequenceCheckbox:Editor Utility Check Box Object Reference)) -> (ReturnValue:ECheckBoxState Enum=Unchecked)
   ) -> (Unchecked:Exec, Checked:Exec, Undetermined:Exec) {is=title K2Node_SwitchEnum}
   SetSelected!(execute=@n2.Checked, self, Selected:bool=true) -> (then)
   SetSelected(execute=@n2.Unchecked, Selected=false)
   n3 = "On Check State Changed (LevelSequenceCheckbox)"!() -> (OutputDelegate, then, bIsChecked:bool) {is=title K2Node_ComponentBoundEvent}
   > SetSelected(self=Self-Reference(), Selected=@n3.bIsChecked)
   n4 = "On Entry Generated (AnimSequencesListView)"!() -> (OutputDelegate, then, Widget:User Widget Object Reference) {is=title K2Node_ComponentBoundEvent}
   n5 = "On Entry Initialized (AnimMontagesListView)"!() -> (OutputDelegate, then, Item:Object Reference, Widget:User Widget Object Reference) {is=title K2Node_ComponentBoundEvent}
   n6 = "On Entry Generated (AnimMontagesListView)"!() -> (OutputDelegate, then, Widget:User Widget Object Reference) {is=title K2Node_ComponentBoundEvent}
   UpdateWidget() {is=title K2Node_CustomEvent}
   > n7 = Sequence!(execute) -> (then_0, then_1, then_2:Exec, then_3:Exec, then_4:Exec, then_5:Exec, then_6:Exec, then_7:Exec)
   n8 = GetObjectName(Object=Get Current Level Sequence() -> (Current Level Sequence:Level Sequence Object Reference))
   Set Text(execute=@n7.then_1, Text="To Text (String)"(InString=@n8), self=Get LevelSequenceText() -> (LevelSequenceText:Text Object Reference))
   n9 = Get Source LS Replaceable() -> (Source LS Replaceable:String)
   n10 = Get Target LS Replaceable() -> (Target LS Replaceable:String)
   > Set Text(
    Text="To Text (String)"(InString=Replace(SourceString=@n8, From=@n9, To=@n10, SearchCase=CaseSensitive))
    self=Get TargetLevelSequenceText() -> (TargetLevelSequenceText:Text Object Reference)
   )
   SetCheckedState!(
    execute=@n7.then_5
    self:Check Box Object Reference=Get LevelSequenceCheckbox()
    InCheckedState:ECheckBoxState Enum=Select(Option 0:ECheckBoxState Enum=Unchecked, Option 1:ECheckBoxState Enum=Checked, Index=Get Selected() -> (Selected:bool)) -> (ReturnValue:ECheckBoxState Enum)
   ) -> (then)
   > Set Border Color!(
    execute
    self
    Border:Border Object Reference=Get BackgroundBorder() -> (BackgroundBorder:Border Object Reference)
    A:Linear Color Structure=(R=0.0176,G=0.0176,B=0.0176,A=1)
    B:Linear Color Structure
    Selected:bool=Get Selected()
   ) -> (then) {is=title K2Node_CallFunction}
   > SetVisibility(
    self=Get TargetLSTextBox() -> (TargetLSTextBox:Vertical Box Object Reference)
    InVisibility=Select(Option 0:ESlateVisibility Enum=Collapsed, Option 1:ESlateVisibility Enum=Visible, Index=Get Selected()) -> (ReturnValue:ESlateVisibility Enum)
   )
   n11 = For Each Loop(
    Exec=@n7.then_6
    Array:Array of User Widget Object References=GetDisplayedEntryWidgets!(self:List View Base Object Reference=Get AnimSequencesListView() -> (AnimSequencesListView:Editor Utility List View Object Reference)) -> (ReturnValue:Array of User Widget Object References)
   ) -> (Array Element:User Widget Object Reference)
   > n12 = Cast To EUW_AnimBatchDup_AS_Item!(execute, Object=@n11) -> (then, CastFailed, AsEUW Anim Batch Dup AS Item:EUW Anim Batch Dup AS Item Object Reference) {is=title K2Node_DynamicCast}
   > Update Widget!(execute, self:EUW Anim Batch Dup AS Item Object Reference=@n12, Selected:bool=Get Selected()) -> (then)
   ClearListItems(execute=@n7.then_2, self=Get AnimSequencesListView())
   > n13 = For Each Loop(Array:Array of Strings=Get Linked AnimSeq Names from LevelSeq!(Level Sequence:Level Sequence Object Reference=Get Current Level Sequence()) -> (AnimSequences Names:Array of Strings) {is=title K2Node_CallFunction}) -> (Array Element:String)
   > n14 = Construct Object from Class!(
    execute
    Class=/AnimInEdTools/AnimBatchDuplicate/Data/BP_AnimBatchDup_AnimSeq_Item.BP_AnimBatchDup_AnimSeq_Item_C
    self
    Animation Sequence:Anim Sequence Object Reference
    Animation Montages:Array of BP Anim Batch Dup Anim Mon Item Object References
    Anim Sequence Name:String=Split(SourceString=@n13, InStr=., SearchCase=CaseSensitive, SearchDir=FromStart).RightS
   ) -> (then, ReturnValue:BP Anim Batch Dup Anim Seq Item Object Reference)
   > AddItem!(execute, self:List View Object Reference=Get AnimSequencesListView(), Item:Object Reference=@n14) -> (then)
   > n15 = For Each Loop(Array=Get AnimMontages Names from AnimSequence Name!(Anim Sequence Path:String=@n13) -> (Anim Montage Names:Array of Strings) {is=title K2Node_CallFunction})
   > Add Unique(TargetArray=Get AnimMontageNames() -> (AnimMontageNames:Array of Strings), NewItem=@n15)
   ClearListItems(execute=@n7.then_3, self=Get AnimMontagesListView() -> (AnimMontagesListView:Editor Utility List View Object Reference))
   > n16 = For Each Loop(Array=Get AnimMontageNames())
   > n17 = Construct Object from Class!(
    execute
    Class=/AnimInEdTools/AnimBatchDuplicate/Data/BP_AnimBatchDup_AnimMon_Item.BP_AnimBatchDup_AnimMon_Item_C
    self
    Animation Montage:Anim Montage Object Reference
    Anim Montage Name:String=@n16
   ) -> (then, ReturnValue:BP Anim Batch Dup Anim Mon Item Object Reference)
   > AddItem(self=Get AnimMontagesListView(), Item=@n17)
   n18 = For Each Loop(Exec=@n7.then_7, Array:Array of User Widget Object References=GetDisplayedEntryWidgets(self=Get AnimMontagesListView())) -> (Array Element:User Widget Object Reference)
   > n19 = Cast To EUW_AnimBatchDup_AM_Item!(execute, Object=@n18) -> (then, CastFailed, AsEUW Anim Batch Dup AM Item:EUW Anim Batch Dup AM Item Object Reference) {is=title K2Node_DynamicCast}
   > UpdateWidget!(execute, self:EUW Anim Batch Dup AM Item Object Reference=@n19, Selected:bool=Get Selected()) -> (then) {is=title K2Node_CallFunction}
   Clear(execute=@n7.then_0, TargetArray=Get AnimMontageNames())
   > n20 = Cast To BP_AnimBatchDup_LS_Item!(execute, Object=Get List Item Object() -> (List Item Object:Object Reference)) -> (then, CastFailed, AsBP Anim Batch Dup LS Item:BP Anim Batch Dup LS Item Object Reference) {is=title K2Node_DynamicCast}
   > Set Selected(Selected:bool=Get Selected!(self:BP Anim Batch Dup LS Item Object Reference=@n20) -> (Selected)) -> (Output_Get:bool)
   > Set Current Level Sequence(Current Level Sequence:Level Sequence Object Reference=Get Level Sequence!(self:BP Anim Batch Dup LS Item Object Reference=@n20) -> (Level Sequence:Level Sequence Object Reference)) -> (Output_Get:Level Sequence Object Reference)
   n21 = UpdateTargets!() -> (OutputDelegate, then, Source LS:String, Target LS:String, Source AS:String, Target AS:String, Source AM:String, Target AM:String) {is=title K2Node_CustomEvent}
   > n22 = Sequence!(execute) -> (then_0, then_1, then_2)
   > Set Source LS Replaceable(Source LS Replaceable:String=@n21.Source LS) -> (Output_Get:String)
   > Set Target LS Replaceable(Target LS Replaceable:String=@n21.Target LS) -> (Output_Get:String)
   > n23 = Cast To BP_AnimBatchDup_LS_Item(Object=Get List Item Object())
   n24 = GetDisplayName(Object=Get Level Sequence(self=@n23))
   n25 = Get Source LS Replaceable()
   n26 = Get Target LS Replaceable()
   > Set Text(
    Text="To Text (String)"(InString=Replace(SourceString=@n24, From=@n25, To=@n26, SearchCase=CaseSensitive))
    self=Get TargetLevelSequenceText()
   )
   n27 = For Each Loop(Exec=@n22.then_1, Array=GetDisplayedEntryWidgets(self=Get AnimSequencesListView()))
   > n28 = Cast To EUW_AnimBatchDup_AS_Item(Object=@n27)
   > UpdateTargets!(
    execute
    self:EUW Anim Batch Dup AS Item Object Reference=@n28
    Source AS:String=@n21.Source AS
    Target AS:String=@n21.Target AS
    Source AM:String
    Target AM:String
   ) -> (then) {is=title K2Node_CallFunction}
   n29 = For Each Loop(Exec=@n22.then_2, Array=GetDisplayedEntryWidgets(self=Get AnimMontagesListView()))
   > n30 = Cast To EUW_AnimBatchDup_AM_Item(Object=@n29)
   > UpdateTargets!(execute, self:EUW Anim Batch Dup AM Item Object Reference=@n30, Source AM=@n21.Source AM, Target AM=@n21.Target AM) -> (then)
   n31 = SetSelected!() -> (OutputDelegate, then, Selected:bool=~) {is=title K2Node_CustomEvent}
   > n32 = Sequence()
   > SetCheckedState(
    self=Get LevelSequenceCheckbox()
    InCheckedState=Select(Option 0:ECheckBoxState Enum=Unchecked, Option 1:ECheckBoxState Enum=Checked, Index=@n31.Selected) -> (ReturnValue:ECheckBoxState Enum)
   )
   > n33 = Cast To BP_AnimBatchDup_LS_Item(Object=GetListItemObject!(UserObjectListEntry:User Object List Entry Interface) -> (ReturnValue:Object Reference))
   > Set Selected!(execute, Selected=@n31.Selected, self:BP Anim Batch Dup LS Item Object Reference=@n33) -> (then, Output_Get)
   > Set Border Color(Border=Get BackgroundBorder(), A=(R=0.0176,G=0.0176,B=0.0176,A=1), Selected=@n31.Selected)
   > SetVisibility(
    self=Get TargetLSTextBox()
    InVisibility=Select(Option 0:ESlateVisibility Enum=Collapsed, Option 1:ESlateVisibility Enum=Visible, Index=@n31.Selected) -> (ReturnValue:ESlateVisibility Enum)
   )
   n34 = For Each Loop(Exec=@n32.then_1, Array=GetDisplayedEntryWidgets(self=Get AnimSequencesListView()))
   > n35 = Cast To EUW_AnimBatchDup_AS_Item(Object=@n34)
   > Update Widget(self=@n35, Selected=@n31.Selected)
   n36 = For Each Loop(Exec=@n32.then_2, Array=GetDisplayedEntryWidgets(self=Get AnimMontagesListView()))
   > n37 = Cast To EUW_AnimBatchDup_AM_Item(Object=@n36)
   > UpdateWidget(self=@n37, Selected=@n31.Selected)
   Clear List Views() {is=title K2Node_CustomEvent}
   > ClearListItems(self=Get AnimSequencesListView())
   > ClearListItems(self=Get AnimMontagesListView())
   Event Construct()
   > SetVisibility(self=Get TargetLSTextBox(), InVisibility=Collapsed)
   > SetVisibility(self=Get LSAlreadyExistsText() -> (LSAlreadyExistsText:Text Object Reference), InVisibility=Collapsed)
   n38 = Event OnListItemObjectSet()
   > Set List Item Object(List Item Object:Object Reference=@n38.ListItemObject) -> (Output_Get:Object Reference)
   > UpdateWidget!(execute, self:Self Object Reference) -> (then)
  ]}
 Get AnimMontages Names from AnimSequence Name {nodes [
   Get AnimMontages Names from AnimSequence Name!() -> (then, Anim Sequence Path:String) {is=title K2Node_FunctionEntry}
   > n1 = Sequence!(execute) -> (then_0, then_1)
   > n2 = FindAssetData!(execute, AssetPath:String=Get Anim Sequence Path() -> (Anim Sequence Path:String)) -> (then, ReturnValue_PackageName:Name, ReturnValue_PackagePath:Name, ReturnValue_AssetName:Name, ReturnValue_AssetClassPath_PathString:String)
   > n3 = Get Referencers!(
    execute
    self:Asset Registry Interface=GetAssetRegistry()
    PackageName:Name=@n2
    ReferenceOptions:"Asset Registry Dependency Options Structure (by ref)"=Make Asset Registry Dependency Options(bIncludeSoftPackageReferences=True)
   ) -> (then, OutReferencers:Array of Names, ReturnValue:bool) {is=title K2Node_CallFunction}
   > n4 = For Each Loop(Array:Array of Names=@n3) -> (Array Element:Name)
   > n5 = FindAssetData(AssetPath="To String (Name)"(InName=@n4))
   > Branch(Condition=Contains(SearchIn=@n5.ReturnValue_AssetClassPath_PathString, Substring=Engine.AnimMontage))
   > Add Unique(
    TargetArray=Get Local Anim Montage Names() -> (Local Anim Montage Names:Array of Strings)
    NewItem="To String (Name)"(InName=@n5.ReturnValue_AssetName)
   )
   Return Node!(execute=@n1.then_1, Anim Montage Names:Array of Strings=Get Local Anim Montage Names())
  ]}
 Get AnimMontages from AnimSequence {nodes [
   Get AnimMontages from AnimSequence!() -> (then, Anim Sequence:Anim Sequence Object Reference) {is=title K2Node_FunctionEntry}
   > n1 = Sequence()
   n2 = GetPackageForObject!(Object:Object Reference=Get Anim Sequence() -> (Anim Sequence:Anim Sequence Object Reference)) -> (ReturnValue:Package Object Reference)
   > n3 = Get Referencers(
    self=GetAssetRegistry()
    PackageName=String To Name(InString=GetObjectName(Object=@n2))
    ReferenceOptions=Make Asset Registry Dependency Options(bIncludeSoftPackageReferences=True)
   )
   > n4 = For Each Loop(Array=@n3)
   > n5 = LoadAsset(AssetPath="To String (Name)"(InName=@n4))
   > n6 = Cast To AnimMontage!(execute, Object=@n5) -> (then, CastFailed, AsAnim Montage:Anim Montage Object Reference) {is=title K2Node_DynamicCast}
   > Add(
    TargetArray:Array of Anim Montage Object References=Get Local Anim Montages() -> (Local Anim Montages:Array of Anim Montage Object References)
    NewItem:"Anim Montage Object Reference (by ref)"=@n6
   ) {is=title K2Node_CallArrayFunction}
   Return Node!(execute=@n1.then_1, Anim Montages:Array of Anim Montage Object References=Get Local Anim Montages())
  ]}
 Get Linked AnimSeq Names from LevelSeq {nodes [
   n1 = Get Linked AnimSeq Names from LevelSeq!() -> (then, Level Sequence:Level Sequence Object Reference) {is=title K2Node_FunctionEntry}
   > n2 = Sequence()
   > n3 = GetAnimSequenceLinkFromLevelSequence!(execute, InLevelSequence:Level Sequence Object Reference=@n1) -> (then, ReturnValue:Level Sequence Anim Sequence Link Object Reference)
   > n4 = For Each Loop(Array:Array of Level Sequence Anim Sequence Link Item Structures=Get AnimSequenceLinks!(self:Level Sequence Anim Sequence Link Object Reference=@n3) -> (AnimSequenceLinks:Array of Level Sequence Anim Sequence Link Item Structures)) -> (Array Element:Level Sequence Anim Sequence Link Item Structure)
   n5 = Break Level Sequence Anim Sequence Link Item!(LevelSequenceAnimSequenceLinkItem:"Level Sequence Anim Sequence Link Item Structure (by ref)"=@n4) -> (
    SkelTrackGuid:Guid Structure
    PathToAnimSequence_PathString:String
    bAutoBake:bool=~
    bExportTransforms:bool=~
    bExportMorphTargets:bool=~
    bExportAttributeCurves:bool=~
    bExportMaterialCurves:bool=~
    bSkipCurvesWithZeroValue:bool=~
    bSetRetargetSourceAsset:bool=~
    Interpolation:EAnimInterpolationType Enum
    CurveInterpolation:ERichCurveInterpMode Enum
    bRecordInWorldSpace:bool=~
    bEvaluateAllSkeletalMeshComponents:bool=~
    IncludeAnimationNames:Array of Strings
    ExcludeAnimationNames:Array of Strings
    bRemoveExcludedCurves:bool=~
    WarmUpFrames:Frame Number Structure
    DelayBeforeStart:Frame Number Structure
    bUseCustomTimeRange:bool=~
    CustomStartFrame:Frame Number Structure
    CustomEndFrame:Frame Number Structure
    CustomDisplayRate:Frame Rate Structure
    bUseCustomFrameRate:bool=~
    CustomFrameRate:Frame Rate Structure
   ) {is=title K2Node_BreakStruct}
   > n6 = DoesAssetExist(AssetPath=@n5.PathToAnimSequence_PathString)
   > Branch(Condition=@n6)
   > Add(
    TargetArray:Array of Strings=Get Local AnimSequences Names() -> (Local AnimSequences Names:Array of Strings)
    NewItem:"String (by ref)"=@n5.PathToAnimSequence_PathString
   )
   Return Node!(execute=@n2.then_1, AnimSequences Names:Array of Strings=Get Local AnimSequences Names())
  ]}
 Get Linked AnimSeq from LevelSeq {nodes [
   n1 = Get Linked AnimSeq from LevelSeq!() -> (then, Level Sequence:Level Sequence Object Reference) {is=title K2Node_FunctionEntry}
   > n2 = Sequence()
   > n3 = GetAnimSequenceLinkFromLevelSequence(InLevelSequence=@n1)
   > n4 = For Each Loop(Array=Get AnimSequenceLinks(self=@n3))
   n5 = Break Level Sequence Anim Sequence Link Item(LevelSequenceAnimSequenceLinkItem=@n4) -> (
    bAutoBake=~
    bExportTransforms=~
    bExportMorphTargets=~
    bExportAttributeCurves=~
    bExportMaterialCurves=~
    bSkipCurvesWithZeroValue=~
    bSetRetargetSourceAsset=~
    bRecordInWorldSpace=~
    bEvaluateAllSkeletalMeshComponents=~
    bRemoveExcludedCurves=~
    bUseCustomTimeRange=~
    bUseCustomFrameRate=~
   )
   > n6 = DoesAssetExist(AssetPath=@n5.PathToAnimSequence_PathString)
   > Branch(Condition=@n6)
   > n7 = LoadAsset(AssetPath=@n5.PathToAnimSequence_PathString)
   > n8 = Cast To AnimSequence!(execute, Object=@n7) -> (then, CastFailed, AsAnim Sequence:Anim Sequence Object Reference)
   > Add(
    TargetArray:Array of Anim Sequence Object References=Get Local AnimSeq() -> (Local AnimSeq:Array of Anim Sequence Object References)
    NewItem:"Anim Sequence Object Reference (by ref)"=@n8
   )
   Return Node!(execute=@n2.then_1, Animation Sequences:Array of Anim Sequence Object References=Get Local AnimSeq())
  ]}
 Set Border Color {nodes [
   n1 = Set Border Color!() -> (then, Border:Border Object Reference, A:Linear Color Structure, B:Linear Color Structure, Selected:bool=~) {is=title K2Node_FunctionEntry}
   n2 = SelectColor(A=@n1.A, B=@n1.B, bPickA=@n1.Selected)
   > SetBrush(
    self:Border Object Reference=@n1
    InBrush=Make Slate Brush!(
     TintColor:Slate Color Structure=Make Slate Color(SpecifiedColor=@n2)
     DrawAs:ESlateBrushDrawType Enum=Image
     Tiling:ESlateBrushTileType Enum=NoTile
     Mirroring:ESlateBrushMirrorType Enum=NoMirror
     ImageSize:Vector 2D Structure=(X=32,Y=32)
     Margin:Margin Structure=(Left=0,Top=0,Right=0,Bottom=0)
     ResourceObject:Object Reference
     OutlineSettings:Slate Brush Outline Settings Structure=(CornerRadii=(X=0,Y=0,Z=0,W=0),Color=(SpecifiedColor=(R=0,G=0,B=0,A=0),ColorUseRule=UseColor_Specified),Width=0,RoundingType=HalfHeightRadius,bUseBrushTransparency=False)
    ) -> (SlateBrush:Slate Brush Structure) {is=title K2Node_MakeStruct}
   )
  ]}
]
variables [
 AnimMontageNames {type="array<string>", default []}
 AnimMontages {type="array<AnimMontage>", default []}
 Current Level Sequence {type=LevelSequence, default=null}
 List Item Object {type=Object, default=null}
 Selected {type=boolean}
 Source LS Replaceable {type=string, default=""}
 Target LS Replaceable {type=string, default=""}
]
widgets {
 info {rootWidgetClass=&StackBox, widgetCount=14}
 widgets [
  StackBox_0 {widgetClassPath=&StackBox, widget=StackBox_0}
  BackgroundBorder {
   widgetClassPath=&Border
   bIsVariable=true
   parent=StackBox_0
   widget=BackgroundBorder
   properties {
    background {
     bIsDynamicallyLoaded=false
     drawAs=Image
     imageType=NoImage
     mirroring=NoMirror
     resourceName=None
     resourceObject=null
     tiling=NoTile
     imageSize=<32 32>
     margin=<ltrb 0 0 0 0>
     outlineSettings {
      bUseBrushTransparency=false
      roundingType=HalfHeightRadius
      width=0
      color {colorUseRule=UseColor_Specified, specifiedColor=<rgba 0 0 0 0>}
      cornerRadii=<0 0 0 0>
     }
     tintColor {colorUseRule=UseColor_Specified, specifiedColor=<rgba 0.0103 0.0103 0.0103 1>}
     uVRegion {bIsValid=false, max=<0 0>, min=<0 0>}
    }
    padding=<ltrb 0 3 0 3>
   }
   slot {class=StackBoxSlot, properties {padding=<ltrb 0 2 0 2>, size {sizeRule=Fill, value=1}}}
  }
  HorizontalBox_0 {
   widgetClassPath=&HorizontalBox
   parent=BackgroundBorder
   widget=HorizontalBox_0
   slot {class=BorderSlot, properties {padding=<ltrb 0 3 0 3>}}
  }
  LevelSequenceCheckbox {
   widgetClassPath=&EditorUtilityCheckBox
   bIsVariable=true
   parent=HorizontalBox_0
   widget=LevelSequenceCheckbox
   properties {
    horizontalAlignment=HAlign_Center
    widgetStyle {
     checkBoxType=CheckBox
     backgroundHoveredImage {
      bIsDynamicallyLoaded=false
      drawAs=RoundedBox
      imageType=NoImage
      mirroring=NoMirror
      resourceName=None
      resourceObject=null
      tiling=NoTile
      imageSize=<18 18>
      margin=<ltrb 0 0 0 0>
      outlineSettings {
       bUseBrushTransparency=false
       roundingType=FixedRadius
       width=1
       color {colorUseRule=UseColor_ColorTable, specifiedColor=<rgba 0 0 0 0>}
       cornerRadii=<3 3 3 3>
      }
      tintColor {colorUseRule=UseColor_Specified, specifiedColor=<rgba 0.0048 0.0048 0.0048 1>}
      uVRegion {bIsValid=false, max=<0 0>, min=<0 0>}
     }
     backgroundImage {
      bIsDynamicallyLoaded=false
      drawAs=RoundedBox
      imageType=NoImage
      mirroring=NoMirror
      resourceName=None
      resourceObject=null
      tiling=NoTile
      imageSize=<18 18>
      margin=<ltrb 0 0 0 0>
      outlineSettings {
       bUseBrushTransparency=false
       roundingType=FixedRadius
       width=1
       color {colorUseRule=UseColor_ColorTable, specifiedColor=<rgba 0 0 0 0>}
       cornerRadii=<3 3 3 3>
      }
      tintColor {colorUseRule=UseColor_Specified, specifiedColor=<rgba 0.0048 0.0048 0.0048 1>}
      uVRegion {bIsValid=false, max=<0 0>, min=<0 0>}
     }
     backgroundPressedImage {
      bIsDynamicallyLoaded=false
      drawAs=RoundedBox
      imageType=NoImage
      mirroring=NoMirror
      resourceName=None
      resourceObject=null
      tiling=NoTile
      imageSize=<18 18>
      margin=<ltrb 0 0 0 0>
      outlineSettings {
       bUseBrushTransparency=false
       roundingType=FixedRadius
       width=1
       color {colorUseRule=UseColor_ColorTable, specifiedColor=<rgba 0 0 0 0>}
       cornerRadii=<3 3 3 3>
      }
      tintColor {colorUseRule=UseColor_Specified, specifiedColor=<rgba 0.0048 0.0048 0.0048 1>}
      uVRegion {bIsValid=false, max=<0 0>, min=<0 0>}
     }
     borderBackgroundColor {colorUseRule=UseColor_Specified, specifiedColor=<rgba 1 1 1 1>}
     checkedForeground {colorUseRule=UseColor_Specified, specifiedColor=<rgba 1 1 1 1>}
     checkedHoveredForeground {colorUseRule=UseColor_Specified, specifiedColor=<rgba 1 1 1 1>}
     checkedHoveredImage {
      bIsDynamicallyLoaded=false
      drawAs=Image
      imageType=Vector
      mirroring=NoMirror
      resourceName=../../../Engine/Content/Slate/Starship/CoreWidgets/CheckBox/check.svg
      resourceObject=null
      tiling=NoTile
      imageSize=<16 16>
      margin=<ltrb 0 0 0 0>
      outlineSettings {
       bUseBrushTransparency=false
       roundingType=HalfHeightRadius
       width=0
       color {colorUseRule=UseColor_Specified, specifiedColor=<rgba 0 0 0 0>}
       cornerRadii=<0 0 0 0>
      }
      tintColor {colorUseRule=UseColor_Specified, specifiedColor=<rgba 0.0044 0.2384 1 1>}
      uVRegion {bIsValid=false, max=<0 0>, min=<0 0>}
     }
     checkedImage {
      bIsDynamicallyLoaded=false
      drawAs=Image
      imageType=Vector
      mirroring=NoMirror
      resourceName=../../../Engine/Content/Slate/Starship/CoreWidgets/CheckBox/check.svg
      resourceObject=null
      tiling=NoTile
      imageSize=<16 16>
      margin=<ltrb 0 0 0 0>
      outlineSettings {
       bUseBrushTransparency=false
       roundingType=HalfHeightRadius
       width=0
       color {colorUseRule=UseColor_Specified, specifiedColor=<rgba 0 0 0 0>}
       cornerRadii=<0 0 0 0>
      }
      tintColor {colorUseRule=UseColor_Specified, specifiedColor=<rgba 0 0.162 0.7454 1>}
      uVRegion {bIsValid=false, max=<0 0>, min=<0 0>}
     }
     checkedPressedForeground {colorUseRule=UseColor_Specified, specifiedColor=<rgba 1 1 1 1>}
     checkedPressedImage {
      bIsDynamicallyLoaded=false
      drawAs=Image
      imageType=Vector
      mirroring=NoMirror
      resourceName=../../../Engine/Content/Slate/Starship/CoreWidgets/CheckBox/check.svg
      resourceObject=null
      tiling=NoTile
      imageSize=<16 16>
      margin=<ltrb 0 0 0 0>
      outlineSettings {
       bUseBrushTransparency=false
       roundingType=HalfHeightRadius
       width=0
       color {colorUseRule=UseColor_Specified, specifiedColor=<rgba 0 0 0 0>}
       cornerRadii=<0 0 0 0>
      }
      tintColor {colorUseRule=UseColor_Specified, specifiedColor=<rgba 0 0.0802 0.3515 1>}
      uVRegion {bIsValid=false, max=<0 0>, min=<0 0>}
     }
     checkedSlateSound {resourceObject=null}
     foregroundColor {colorUseRule=UseColor_Specified, specifiedColor=<rgba 1 1 1 1>}
     hoveredForeground {colorUseRule=UseColor_Specified, specifiedColor=<rgba 1 1 1 1>}
     hoveredSlateSound {resourceObject=null}
     padding=<ltrb 0 0 0 0>
     pressedForeground {colorUseRule=UseColor_Specified, specifiedColor=<rgba 1 1 1 1>}
     uncheckedHoveredImage {
      bIsDynamicallyLoaded=false
      drawAs=NoDrawType
      imageType=NoImage
      mirroring=NoMirror
      resourceName=None
      resourceObject=null
      tiling=NoTile
      imageSize=<0 0>
      margin=<ltrb 0 0 0 0>
      outlineSettings {
       bUseBrushTransparency=false
       roundingType=HalfHeightRadius
       width=0
       color {colorUseRule=UseColor_Specified, specifiedColor=<rgba 0 0 0 0>}
       cornerRadii=<0 0 0 0>
      }
      tintColor {colorUseRule=UseColor_Specified, specifiedColor=<rgba 1 1 1 1>}
      uVRegion {bIsValid=false, max=<0 0>, min=<0 0>}
     }
     uncheckedImage {
      bIsDynamicallyLoaded=false
      drawAs=NoDrawType
      imageType=NoImage
      mirroring=NoMirror
      resourceName=None
      resourceObject=null
      tiling=NoTile
      imageSize=<0 0>
      margin=<ltrb 0 0 0 0>
      outlineSettings {
       bUseBrushTransparency=false
       roundingType=HalfHeightRadius
       width=0
       color {colorUseRule=UseColor_Specified, specifiedColor=<rgba 0 0 0 0>}
       cornerRadii=<0 0 0 0>
      }
      tintColor {colorUseRule=UseColor_Specified, specifiedColor=<rgba 1 1 1 1>}
      uVRegion {bIsValid=false, max=<0 0>, min=<0 0>}
     }
     uncheckedPressedImage {
      bIsDynamicallyLoaded=false
      drawAs=NoDrawType
      imageType=NoImage
      mirroring=NoMirror
      resourceName=None
      resourceObject=null
      tiling=NoTile
      imageSize=<0 0>
      margin=<ltrb 0 0 0 0>
      outlineSettings {
       bUseBrushTransparency=false
       roundingType=HalfHeightRadius
       width=0
       color {colorUseRule=UseColor_Specified, specifiedColor=<rgba 0 0 0 0>}
       cornerRadii=<0 0 0 0>
      }
      tintColor {colorUseRule=UseColor_Specified, specifiedColor=<rgba 1 1 1 1>}
      uVRegion {bIsValid=false, max=<0 0>, min=<0 0>}
     }
     uncheckedSlateSound {resourceObject=null}
     undeterminedForeground {colorUseRule=UseColor_Specified, specifiedColor=<rgba 1 1 1 1>}
     undeterminedHoveredImage {
      bIsDynamicallyLoaded=false
      drawAs=Image
      imageType=Vector
      mirroring=NoMirror
      resourceName=../../../Engine/Content/Slate/Starship/CoreWidgets/CheckBox/indeterminate.svg
      resourceObject=null
      tiling=NoTile
      imageSize=<16 16>
      margin=<ltrb 0 0 0 0>
      outlineSettings {
       bUseBrushTransparency=false
       roundingType=HalfHeightRadius
       width=0
       color {colorUseRule=UseColor_Specified, specifiedColor=<rgba 0 0 0 0>}
       cornerRadii=<0 0 0 0>
      }
      tintColor {colorUseRule=UseColor_Specified, specifiedColor=<rgba 0.0044 0.2384 1 1>}
      uVRegion {bIsValid=false, max=<0 0>, min=<0 0>}
     }
     undeterminedImage {
      bIsDynamicallyLoaded=false
      drawAs=Image
      imageType=Vector
      mirroring=NoMirror
      resourceName=../../../Engine/Content/Slate/Starship/CoreWidgets/CheckBox/indeterminate.svg
      resourceObject=null
      tiling=NoTile
      imageSize=<16 16>
      margin=<ltrb 0 0 0 0>
      outlineSettings {
       bUseBrushTransparency=false
       roundingType=HalfHeightRadius
       width=0
       color {colorUseRule=UseColor_Specified, specifiedColor=<rgba 0 0 0 0>}
       cornerRadii=<0 0 0 0>
      }
      tintColor {colorUseRule=UseColor_Specified, specifiedColor=<rgba 0 0.162 0.7454 1>}
      uVRegion {bIsValid=false, max=<0 0>, min=<0 0>}
     }
     undeterminedPressedImage {
      bIsDynamicallyLoaded=false
      drawAs=Image
      imageType=Vector
      mirroring=NoMirror
      resourceName=../../../Engine/Content/Slate/Starship/CoreWidgets/CheckBox/indeterminate.svg
      resourceObject=null
      tiling=NoTile
      imageSize=<16 16>
      margin=<ltrb 0 0 0 0>
      outlineSettings {
       bUseBrushTransparency=false
       roundingType=HalfHeightRadius
       width=0
       color {colorUseRule=UseColor_Specified, specifiedColor=<rgba 0 0 0 0>}
       cornerRadii=<0 0 0 0>
      }
      tintColor {colorUseRule=UseColor_Specified, specifiedColor=<rgba 0 0.0802 0.3515 1>}
      uVRegion {bIsValid=false, max=<0 0>, min=<0 0>}
     }
    }
   }
   slot {class=HorizontalBoxSlot, properties {verticalAlignment=VAlign_Center, padding=<ltrb 10 1 10 0>}}
  }
  VerticalBox_0 {
   widgetClassPath=&VerticalBox
   parent=HorizontalBox_0
   widget=VerticalBox_0
   slot {class=HorizontalBoxSlot, properties {padding=<ltrb 10 0 0 0>, size {sizeRule=Fill, value=1}}}
  }
  HorizontalBox_1 {widgetClassPath=&HorizontalBox, parent=VerticalBox_0, widget=HorizontalBox_1, slot {class=VerticalBoxSlot}}
  LevelSequenceText {
   widgetClassPath=&TextBlock
   bIsVariable=true
   parent=HorizontalBox_1
   widget=LevelSequenceText
   properties {
    text=Text Block
    colorAndOpacity {colorUseRule=UseColor_Specified, specifiedColor=<rgba 0.5776 0.0802 0.0802 1>}
    font {
     bForceMonospaced=false
     bMaterialIsStencil=false
     fontMaterial=null
     fontObject=&Roboto
     letterSpacing=0
     monospacedWidth=1
     size=12
     skewAmount=0
     typefaceFontName=Bold
     outlineSettings {
      bApplyOutlineToDropShadows=false
      bMiteredCorners=false
      bSeparateFillAlpha=false
      outlineMaterial=null
      outlineSize=0
      outlineColor=<rgba 0 0 0 1>
     }
    }
   }
   slot {class=HorizontalBoxSlot, properties {horizontalAlignment=HAlign_Left, verticalAlignment=VAlign_Bottom, size {sizeRule=Fill, value=1}}}
  }
  TargetLSTextBox {
   widgetClassPath=&VerticalBox
   bIsVariable=true
   parent=HorizontalBox_1
   widget=TargetLSTextBox
   slot {class=HorizontalBoxSlot, properties {size {sizeRule=Fill, value=1}}}
  }
  HorizontalBox_2 {
   widgetClassPath=&HorizontalBox
   parent=TargetLSTextBox
   widget=HorizontalBox_2
   slot {class=VerticalBoxSlot, properties {padding=<ltrb 20 0 0 0>}}
  }
  Image_248 {
   widgetClassPath=&Image
   bIsVariable=true
   parent=HorizontalBox_2
   widget=Image_248
   properties {
    brush {
     bIsDynamicallyLoaded=false
     drawAs=Image
     imageType=FullColor
     mirroring=NoMirror
     resourceName=None
     resourceObject=/AnimInEdTools/AnimBatchDuplicate/Icons/Icon_add
     tiling=NoTile
     imageSize=<14 14>
     margin=<ltrb 0 0 0 0>
     outlineSettings {
      bUseBrushTransparency=false
      roundingType=HalfHeightRadius
      width=0
      color {colorUseRule=UseColor_Specified, specifiedColor=<rgba 0 0 0 0>}
      cornerRadii=<0 0 0 0>
     }
     tintColor {colorUseRule=UseColor_Specified, specifiedColor=<rgba 1 1 1 1>}
     uVRegion {bIsValid=false, max=<0 0>, min=<0 0>}
    }
    colorAndOpacity=<rgba 0.016 0.2542 0.4852 1>
   }
   slot {class=HorizontalBoxSlot, properties {horizontalAlignment=HAlign_Center, verticalAlignment=VAlign_Center}}
  }
  TargetLevelSequenceText {
   widgetClassPath=&TextBlock
   bIsVariable=true
   parent=HorizontalBox_2
   widget=TargetLevelSequenceText
   properties {
    text=Text Block
    colorAndOpacity {colorUseRule=UseColor_Specified, specifiedColor=<rgba 0.5776 0.0802 0.0802 1>}
    font {
     bForceMonospaced=false
     bMaterialIsStencil=false
     fontMaterial=null
     fontObject=&Roboto
     letterSpacing=0
     monospacedWidth=1
     size=12
     skewAmount=0
     typefaceFontName=Italic
     outlineSettings {
      bApplyOutlineToDropShadows=false
      bMiteredCorners=false
      bSeparateFillAlpha=false
      outlineMaterial=null
      outlineSize=0
      outlineColor=<rgba 0 0 0 1>
     }
    }
   }
   slot {class=HorizontalBoxSlot, properties {padding=<ltrb 5 0 0 0>}}
  }
  LSAlreadyExistsText {
   widgetClassPath=&TextBlock
   bIsVariable=true
   parent=HorizontalBox_2
   widget=LSAlreadyExistsText
   properties {
    text=* Already Exists *
    colorAndOpacity {colorUseRule=UseColor_Specified, specifiedColor=<rgba 0.5776 0.5776 0.5776 1>}
    font {
     bForceMonospaced=false
     bMaterialIsStencil=false
     fontMaterial=null
     fontObject=&Roboto
     letterSpacing=0
     monospacedWidth=1
     size=12
     skewAmount=0
     typefaceFontName=Italic
     outlineSettings {
      bApplyOutlineToDropShadows=false
      bMiteredCorners=false
      bSeparateFillAlpha=false
      outlineMaterial=null
      outlineSize=0
      outlineColor=<rgba 0 0 0 1>
     }
    }
   }
   slot {class=HorizontalBoxSlot, properties {horizontalAlignment=HAlign_Right, padding=<ltrb 15 0 0 0>, size {sizeRule=Fill, value=1}}}
  }
  AnimSequencesListView {
   widgetClassPath=/Script/Blutility.EditorUtilityListView
   bIsVariable=true
   parent=VerticalBox_0
   widget=AnimSequencesListView
   properties {entryWidgetClass=&/EUW_AnimBatchDup_AS_Item.EUW_AnimBatchDup_AS_Item_C, numDesignerPreviewEntries=1}
   slot {class=VerticalBoxSlot}
  }
  AnimMontagesListView {
   widgetClassPath=&EditorUtilityListView
   bIsVariable=true
   parent=VerticalBox_0
   widget=AnimMontagesListView
   properties {entryWidgetClass=&/EUW_AnimBatchDup_AM_Item.EUW_AnimBatchDup_AM_Item_C, numDesignerPreviewEntries=1}
   slot {class=VerticalBoxSlot}
  }
 ]
}
