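# Editor Utility Widget that batch-duplicates selected Level Sequences from a source Animatable Creature to a target one, with linked animations and montages.
# Source and target paths resolve from each creature, filling tokenized patterns, and asset names get string replacement.
# Montages are found through Asset Registry referencers.
# Existing targets are reused, sequence links are repointed both ways, partly via Python, montage segments are remapped, and everything is saved.
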EditorUtilityWidgetBlueprint /AnimInEdTools/AnimBatchDuplicate/EUW_AnimBatchDuplicate
graphs [
 Define Target Path {nodes [
   Define Target Path!() -> (then, Target Path:String, Object:Object Reference, Subfolder:bool=~, Source:String, Target:String) {is=title K2Node_FunctionEntry}
   > n1 = Sequence()
   n2 = Get Target Path() -> (Target Path:String)
   n3 = SelectString(A=@n2, B=Append(A=@n2, B=/), bPickA=EndsWith(SourceString=@n2, InSuffix=/, SearchCase=CaseSensitive))
   n4 = Get Object() -> (Object:Object Reference)
   n5 = Replace(SourceString=GetObjectName(Object=@n4), From=Get Source() -> (Source:String), To=Get Target() -> (Target:String), SearchCase=CaseSensitive)
   n6 = Get Last Folder Name From Path!(Path:String=Get Object Path String(Object=@n4)) -> (Folder:String) {is=title K2Node_CallFunction}
   > Set Local Path(Local Path:String=SelectString(
     A=Append!(A=@n3, B=/, C:String=@n6, D:String=/, E:String=@n5) -> (ReturnValue)
     B=Append!(A=@n3, B=@n5) -> (ReturnValue)
     bPickA=Get Subfolder() -> (Subfolder:bool)
    )) -> (Output_Get:String)
   Return Node!(execute=@n1.then_1, Path:String=Get Local Path() -> (Local Path:String))
  ]}
 Duplicate Animations {nodes [
   Duplicate Animations!() -> (
    then
    Level Sequence:Level Sequence Object Reference
    LS Target Path:String
    AS Target Path:String
    AM Target Path:String
    Source Name:String
    Target Name:String
    Source AS Replaceable:String
    Target AS Replaceable:String
    Source AM Replaceable:String
    Target AM Replaceable:String
   ) {is=title K2Node_FunctionEntry}
   > n1 = Sequence!(execute) -> (then_0, then_1, then_2:Exec, then_3:Exec, then_4:Exec, then_5:Exec, then_6:Exec, then_7:Exec, then_8:Exec)
   n2 = GetAnimSequenceLinkFromLevelSequence!(
    execute=@n1.then_4
    InLevelSequence:Level Sequence Object Reference=Get Duplicated Level Sequence() -> (Duplicated Level Sequence:Level Sequence Object Reference)
   ) -> (then, ReturnValue:Level Sequence Anim Sequence Link Object Reference)
   > n3 = For Each Loop(Array:Array of Level Sequence Anim Sequence Link Item Structures=Get AnimSequenceLinks!(self:Level Sequence Anim Sequence Link Object Reference=@n2) -> (AnimSequenceLinks:Array of Level Sequence Anim Sequence Link Item Structures)) -> (Array Element:Level Sequence Anim Sequence Link Item Structure)
   > n4 = Execute Python Script!(
    execute
    PythonScript="import unreal\r\n\r\nsoft_path = unreal.SoftObjectPath(anim_seq_path)\r\nlink.set_editor_property(\"path_to_anim_sequence\", soft_path)\r\n\r\nupd_link = link"
    anim_seq_path:String="Get (a copy)"(Array=Get Target AnimSequences Paths() -> (Target AnimSequences Paths:Array of Strings), Dimension 1=@n3.Array Index)
    link:Level Sequence Anim Sequence Link Item Structure=@n3
   ) -> (then, ReturnValue, upd_link:"Level Sequence Anim Sequence Link Item Structure (by ref)") {inputs [anim_seq_path, link], outputs [upd_link]}
   > Add(
    TargetArray:Array of Level Sequence Anim Sequence Link Item Structures=Get Updated Links() -> (Updated Links:Array of Level Sequence Anim Sequence Link Item Structures)
    NewItem:"Level Sequence Anim Sequence Link Item Structure (by ref)"=@n4.upd_link
   ) {is=title K2Node_CallArrayFunction}
   Set AnimSequenceLinks!(
    execute=@n3.Completed
    AnimSequenceLinks:Array of Level Sequence Anim Sequence Link Item Structures=Get Updated Links()
    self:Level Sequence Anim Sequence Link Object Reference=@n2
   ) -> (then, Output_Get:Array of Level Sequence Anim Sequence Link Item Structures)
   n5 = For Each Loop(
    Exec=@n1.then_1
    Array:Array of Anim Sequence Object References=Get Linked AnimSeq from LevelSeq!(
     Level Sequence:Level Sequence Object Reference=Get Level Sequence() -> (Level Sequence:Level Sequence Object Reference)
     __WorldContext:Object Reference
    ) -> (Animation Sequences:Array of Anim Sequence Object References) {is=title K2Node_CallFunction}
   ) -> (Array Element:Anim Sequence Object Reference)
   n6 = Define Target Path!(
    Target Path:String=Get AS Target Path() -> (AS Target Path:String)
    Object:Object Reference=@n5
    Subfolder:bool=Get Preserve Folder Structure AS() -> (Preserve Folder Structure AS:bool)
    Source:String=Get Source AS Replaceable() -> (Source AS Replaceable:String)
    Target:String=Get Target AS Replaceable() -> (Target AS Replaceable:String)
   ) -> (Path:String) {is=title K2Node_CallFunction}
   > n7 = DoesAssetExist(AssetPath=@n6)
   > n8 = Sequence!(execute) -> (then_0, then_1)
   > n9 = Branch(Condition=@n7)
   > n10 = LoadAsset(AssetPath=@n6)
   > n11 = Cast To AnimSequence!(execute, Object=@n10) -> (then, CastFailed, AsAnim Sequence:Anim Sequence Object Reference)
   > Set Current AnimSequence(Current AnimSequence:Anim Sequence Object Reference=@n11) -> (Output_Get:Anim Sequence Object Reference)
   n12 = DuplicateLoadedAsset!(execute=@n9.else, SourceAsset:Object Reference=@n5, DestinationAssetPath:String=@n6) -> (then, ReturnValue:Object Reference)
   > n13 = Cast To AnimSequence(Object=@n12)
   > Set Current AnimSequence(Current AnimSequence=@n13)
   n14 = Get Current AnimSequence() -> (Current AnimSequence:Anim Sequence Object Reference)
   Add(
    execute=@n8.then_1
    TargetArray:Array of Anim Sequence Object References=Get Duplicated Anim Sequences() -> (Duplicated Anim Sequences:Array of Anim Sequence Object References)
    NewItem:"Anim Sequence Object Reference (by ref)"=@n14
   )
   > Add(
    TargetArray:Array of Strings=Get Target AnimSequences Paths()
    NewItem:"String (by ref)"=GetSoftObjectPath!(Object=@n14) -> (ReturnValue_PathString:String)
   )
   > Add!(
    execute
    TargetMap:Map of Anim Sequence Object References to Anim Sequence Object References=Get Source Target AnimSequences() -> (Source Target AnimSequences:Map of Anim Sequence Object References to Anim Sequence Object References)
    Key:"Anim Sequence Object Reference (by ref)"=@n5
    Value:"Anim Sequence Object Reference (by ref)"=@n14
   ) -> (then) {is=title K2Node_CallFunction}
   n15 = Define Target Path(
    Target Path=Get LS Target Path() -> (LS Target Path:String)
    Object=Get Level Sequence()
    Subfolder=Get Preserve Folder Structure LS() -> (Preserve Folder Structure LS:bool)
    Source=Get Source Name() -> (Source Name:String)
    Target=Get Target Name() -> (Target Name:String)
   )
   n16 = DoesAssetExist(execute=@n1.then_0, AssetPath=@n15)
   > n17 = Branch(Condition=@n16)
   > n18 = LoadAsset(AssetPath=@n15)
   > n19 = Cast To LevelSequence!(execute, Object=@n18) -> (then, CastFailed, AsLevel Sequence:Level Sequence Object Reference) {is=title K2Node_DynamicCast}
   > Set Duplicated Level Sequence(Duplicated Level Sequence:Level Sequence Object Reference=@n19) -> (Output_Get:Level Sequence Object Reference)
   n20 = DuplicateLoadedAsset(execute=@n17.else, SourceAsset=Get Level Sequence(), DestinationAssetPath=@n15)
   > n21 = Cast To LevelSequence(Object=@n20)
   > Set Duplicated Level Sequence(Duplicated Level Sequence=@n21)
   Clear(
    execute=@n1.then_2
    TargetArray:Array of Anim Montage Object References=Get Referenced Anim Montages() -> (Referenced Anim Montages:Array of Anim Montage Object References)
   )
   > n22 = For Each Loop(Array=Get Linked AnimSeq from LevelSeq(Level Sequence=Get Level Sequence()))
   n23 = Get Object Path String(Object=@n22)
   > n24 = Get Referencers!(
    execute
    self:Asset Registry Interface=GetAssetRegistry()
    PackageName:Name=String To Name(InString=Split(SourceString=@n23, InStr=., SearchDir=FromStart))
    ReferenceOptions_bIncludeSoftPackageReferences:bool=True
    ReferenceOptions_bIncludeHardPackageReferences:bool=True
    ReferenceOptions_bIncludeGamePackageReferences:bool=True
    ReferenceOptions_bIncludeEditorOnlyPackageReferences:bool=True
    ReferenceOptions_bIncludeSearchableNames:bool=False
    ReferenceOptions_bIncludeSoftManagementReferences:bool=False
    ReferenceOptions_bIncludeHardManagementReferences:bool=False
   ) -> (then, OutReferencers:Array of Names, ReturnValue:bool) {is=title K2Node_CallFunction}
   > n25 = For Each Loop(Array:Array of Names=@n24) -> (Array Element:Name)
   > n26 = LoadAsset(AssetPath="To String (Name)"(InName=@n25))
   > n27 = Cast To AnimMontage!(execute, Object=@n26) -> (then, CastFailed, AsAnim Montage:Anim Montage Object Reference) {is=title K2Node_DynamicCast}
   > Add Unique(TargetArray:Array of Anim Montage Object References=Get Referenced Anim Montages(), NewItem:"Anim Montage Object Reference (by ref)"=@n27)
   n28 = For Each Loop(Exec=@n25.Completed, Array:Array of Anim Montage Object References=Get Referenced Anim Montages()) -> (Array Element:Anim Montage Object Reference)
   n29 = Define Target Path(
    Target Path=Get AM Target Path() -> (AM Target Path:String)
    Object=@n28
    Subfolder=Get Preserve Folder Structure AM() -> (Preserve Folder Structure AM:bool)
    Source=Get Source AM Replaceable() -> (Source AM Replaceable:String)
    Target=Get Target AM Replaceable() -> (Target AM Replaceable:String)
   )
   > n30 = DoesAssetExist(AssetPath=@n29)
   > n31 = Branch(Condition=@n30)
   > n32 = LoadAsset(AssetPath=@n29)
   > n33 = Cast To AnimMontage(Object=@n32)
   > Add(
    TargetArray:Array of Anim Montage Object References=Get Duplicated AnimMontages() -> (Duplicated AnimMontages:Array of Anim Montage Object References)
    NewItem:"Anim Montage Object Reference (by ref)"=@n33
   ) {is=title K2Node_CallArrayFunction}
   n34 = DuplicateLoadedAsset(execute=@n31.else, SourceAsset=@n28, DestinationAssetPath=@n29)
   > n35 = Cast To AnimMontage(Object=@n34)
   > Add(TargetArray=Get Duplicated AnimMontages(), NewItem=@n35)
   n36 = For Each Loop(Exec=@n1.then_3, Array:Array of Anim Sequence Object References=Get Duplicated Anim Sequences()) -> (Array Element:Anim Sequence Object Reference)
   > n37 = GetLevelSequenceLinkFromAnimSequence!(execute, InAnimSequence:Anim Sequence Object Reference=@n36) -> (then, ReturnValue:Anim Sequence Level Sequence Link Object Reference)
   > Set PathToLevelSequence!(
    execute
    PathToLevelSequence:Soft Object Path Structure=GetSoftObjectPath!(Object=Get Duplicated Level Sequence()) -> (ReturnValue:Soft Object Path Structure)
    self:Anim Sequence Level Sequence Link Object Reference=@n37
   ) -> (then, Output_Get:Soft Object Path Structure)
   SaveLoadedAsset(execute=@n1.then_6, AssetToSave=Get Duplicated Level Sequence(), bOnlyIfIsDirty=false)
   > n38 = For Each Loop(Array=Get Duplicated Anim Sequences())
   > SaveLoadedAsset(AssetToSave=@n38, bOnlyIfIsDirty=false)
   n39 = For Each Loop(Exec=@n1.then_7, Array:Array of Anim Montage Object References=Get Duplicated AnimMontages()) -> (Array Element:Anim Montage Object Reference)
   > SaveLoadedAsset(AssetToSave=@n39, bOnlyIfIsDirty=false)
   n40 = For Each Loop(Exec=@n1.then_5, Array=Get Duplicated AnimMontages())
   > n41 = For Each Loop(Array:Array of Names=GetMontageSlotNames!(AnimationMontage:Anim Montage Object Reference=@n40) -> (SlotNames:Array of Names)) -> (Array Element:Name)
   > n42 = Get Sequences from Montage!(execute, AnimMontage:Anim Montage Object Reference=@n40, SlotName:Name=@n41) -> (then, ReturnValue:Array of Anim Sequence Base Object References) {is=title K2Node_CallFunction}
   > n43 = For Each Loop(Array:Array of Anim Sequence Base Object References=@n42) -> (Array Element:Anim Sequence Base Object Reference)
   > n44 = Cast To AnimSequence(Object=@n43)
   > Set Anim Montage Anim Sequence!(
    execute
    AnimMontage:Anim Montage Object Reference=@n40
    SlotName:Name=@n41
    AnimSequence:Anim Sequence Object Reference=Find(
     TargetMap:Map of Anim Sequence Object References to Anim Sequence Object References=Get Source Target AnimSequences()
     Key:"Anim Sequence Object Reference (by ref)"=@n44
    ) -> (Value:Anim Sequence Object Reference)
    SegmentIndex:int=@n43.Array Index
   ) -> (then, ReturnValue:bool) {is=title K2Node_CallFunction}
  ]}
 EventGraph {nodes [
   "On Clicked (DuplicateButton)"() {is=title K2Node_ComponentBoundEvent}
   > n1 = For Each Loop(Array:Array of Object References=GetListItems!(self:List View Object Reference=Get LevelSequenceListView() -> (LevelSequenceListView:Editor Utility List View Object Reference)) -> (ReturnValue:Array of Object References)) -> (Array Element:Object Reference)
   > n2 = Cast To BP_AnimBatchDup_LS_Item!(execute, Object=@n1) -> (then, CastFailed, AsBP Anim Batch Dup LS Item:BP Anim Batch Dup LS Item Object Reference) {is=title K2Node_DynamicCast}
   > Branch(Condition=Get Selected!(self:BP Anim Batch Dup LS Item Object Reference=@n2) -> (Selected:bool))
   n3 = Get Target AC() -> (Target AC:Animatable Creature Object Reference)
   n4 = Translate Source Path If Tokenized!(AC:Animatable Creature Object Reference=@n3, Anim Folder:AnimBatchDup_Folder Enum=NewEnumerator1) -> (Path:String)
   > Duplicate Animations!(
    execute
    self
    Level Sequence:Level Sequence Object Reference=Get Level Sequence!(self:BP Anim Batch Dup LS Item Object Reference=@n2) -> (Level Sequence)
    LS Target Path:String=Translate Source Path If Tokenized(AC=@n3, Anim Folder=NewEnumerator0)
    AS Target Path:String=@n4
    AM Target Path:String=SelectString(A=@n4, B=Translate Source Path If Tokenized(AC=@n3, Anim Folder=NewEnumerator2), bPickA="Equal Exactly (String)"())
    Source Name:String=Get Source LS Replaceable() -> (Source LS Replaceable:String)
    Target Name:String=Get Target LS Replaceable() -> (Target LS Replaceable:String)
    Source AS Replaceable:String=Get Source AS Replaceable()
    Target AS Replaceable:String=Get Target AS Replaceable()
    Source AM Replaceable:String=Get Source AM Replaceable()
    Target AM Replaceable:String=Get Target AM Replaceable()
   ) -> (then) {is=title K2Node_CallFunction}
   "On Property Changed (SourceACPropertyView)"!() -> (OutputDelegate, then, PropertyName:Name=None) {is=title K2Node_ComponentBoundEvent}
   > n5 = Sequence!(execute) -> (then_0, then_1, then_2:Exec)
   > n6 = For Each Loop(Array:Array of User Widget Object References=GetDisplayedEntryWidgets!(self:List View Base Object Reference=Get LevelSequenceListView()) -> (ReturnValue:Array of User Widget Object References)) -> (Array Element:User Widget Object Reference)
   > n7 = Cast To EUW_AnimBatchDup_LS_Item!(execute, Object=@n6) -> (then, CastFailed, AsEUW Anim Batch Dup LS Item:EUW Anim Batch Dup LS Item Object Reference) {is=title K2Node_DynamicCast}
   > Clear List Views(self:EUW Anim Batch Dup LS Item Object Reference=@n7)
   n8 = Get AC() -> (AC:Animatable Creature Object Reference)
   n9 = GetAssetsByPath!(
    execute=@n5.then_2
    self:Asset Registry Interface=GetAssetRegistry()
    PackagePath:Name=String To Name(InString=Translate Source Path If Tokenized(AC=@n8, Anim Folder=NewEnumerator0))
    bRecursive:bool=true
    bIncludeOnlyOnDiskAssets:bool=false
   ) -> (then, OutAssetData:Array of Asset Data Structures, ReturnValue:bool)
   > n10 = For Each Loop!(Exec, Array:Array of Asset Data Structures=@n9) -> (
    LoopBody
    Array Element_PackageName:Name
    Array Element_PackagePath:Name
    Array Element_AssetName:Name
    Array Element_AssetClassPath_PathString:String
    Array Index
    Completed
   )
   > Branch(Condition="Equal Exactly (String)"(A=@n10.Array Element_AssetClassPath_PathString, B=/Script/LevelSequence))
   > n11 = LoadAsset(AssetPath="To String (Name)"(InName=@n10))
   > n12 = Cast To LevelSequence(Object=@n11)
   > n13 = Construct Object from Class!(
    execute
    Class=/AnimInEdTools/AnimBatchDuplicate/Data/BP_AnimBatchDup_LS_Item.BP_AnimBatchDup_LS_Item_C
    self
    Level Sequence:Level Sequence Object Reference=@n12
    Animation Sequences:Array of BP Anim Batch Dup Anim Seq Item Object References
    Level Sequence Name:String
   ) -> (then, ReturnValue:BP Anim Batch Dup LS Item Object Reference)
   > AddItem!(execute, self:List View Object Reference=Get LevelSequenceListView(), Item:Object Reference=@n13) -> (then)
   ClearListItems(execute=@n5.then_1, self=Get LevelSequenceListView())
   > Hide/Show Widget!(
    execute
    self
    Widget:Array of Widget Object References=Make Array("[0]":Widget Object Reference=Get SourcePathVerticalBox() -> (SourcePathVerticalBox:Vertical Box Object Reference)) -> (Array:Array of Widget Object References)
    Hide:bool=false
   ) -> (then)
   > Set Path Texts from AC!(
    execute
    self
    AC:Animatable Creature Object Reference=Get AC()
    LevelSequence:Text Object Reference=Get SourceLevelSequencesText() -> (SourceLevelSequencesText:Text Object Reference)
    AnimSequence:Text Object Reference
    AnimMontage:Text Object Reference
   ) -> (then) {is=title K2Node_CallFunction}
   > SetCheckedState!(
    execute
    self:Check Box Object Reference=Get SelectAllCheckbox() -> (SelectAllCheckbox:Editor Utility Check Box Object Reference)
    InCheckedState:ECheckBoxState Enum=Unchecked
   ) -> (then)
   n14 = Get Target AC()
   n15 = Get AC()
   n16 = NOT Boolean(A="Equal Exactly (String)"(A=GetObjectName(Object=@n15), B=GetObjectName(Object=@n14)))
   n17 = AND Boolean(A=IsValid(Object=Get AC()), B=IsValid(Object=Get Target AC()))
   > Hide/Show Widget(
    Widget=Make Array("[0]"=Get DuplicateButtonBox() -> (DuplicateButtonBox:Vertical Box Object Reference))
    Hide=NOT Boolean(A=AND Boolean(A=@n17, B=@n16))
   )
   n18 = "On Text Changed (TargetAMTextBox)"!() -> (OutputDelegate, then, Text:"Text (by ref)") {is=title K2Node_ComponentBoundEvent}
   > n19 = Set Target AM Replaceable(Target AM Replaceable:String="To String (Text)"(InText=@n18.Text)) -> (Output_Get:String)
   n20 = For Each Loop!(
    Exec=[@n23.then, @n25.then, @n27.then, @n29.then, @n31.then, @n19.then]
    Array:Array of User Widget Object References=GetDisplayedEntryWidgets(self=Get LevelSequenceListView())
   ) -> (LoopBody, Array Element:User Widget Object Reference, Array Index, Completed)
   > n21 = Cast To EUW_AnimBatchDup_LS_Item(Object=@n20)
   n22 = "On Text Changed (SourceAMTextBox)"!() -> (OutputDelegate, then, Text:"Text (by ref)") {is=title K2Node_ComponentBoundEvent}
   > n23 = Set Source AM Replaceable(Source AM Replaceable:String="To String (Text)"(InText=@n22.Text)) -> (Output_Get:String)
   n24 = "On Text Changed (SourceLSTextBox)"!() -> (OutputDelegate, then, Text:"Text (by ref)") {is=title K2Node_ComponentBoundEvent}
   > n25 = Set Source LS Replaceable(Source LS Replaceable:String="To String (Text)"(InText=@n24.Text)) -> (Output_Get:String)
   n26 = "On Text Changed (TargetLSTextBox)"!() -> (OutputDelegate, then, Text:"Text (by ref)") {is=title K2Node_ComponentBoundEvent}
   > n27 = Set Target LS Replaceable(Target LS Replaceable:String="To String (Text)"(InText=@n26.Text)) -> (Output_Get:String)
   n28 = "On Text Changed (SourceASTextBox)"!() -> (OutputDelegate, then, Text:"Text (by ref)") {is=title K2Node_ComponentBoundEvent}
   > n29 = Set Source AS Replaceable(Source AS Replaceable:String="To String (Text)"(InText=@n28.Text)) -> (Output_Get:String)
   n30 = "On Text Changed (TargetASTextBox)"!() -> (OutputDelegate, then, Text:"Text (by ref)") {is=title K2Node_ComponentBoundEvent}
   > n31 = Set Target AS Replaceable(Target AS Replaceable:String="To String (Text)"(InText=@n30.Text)) -> (Output_Get:String)
   UpdateTargets!(
    execute=@n21.then
    self:EUW Anim Batch Dup LS Item Object Reference=@n21
    Source LS:String=@n25
    Target LS:String=@n27
    Source AS:String=@n29
    Target AS:String=@n31
    Source AM:String=@n23
    Target AM:String=@n19
   ) -> (then)
   n32 = "On Check State Changed (SubfolderStructureLSCheckbox)"!() -> (OutputDelegate, then, bIsChecked:bool) {is=title K2Node_ComponentBoundEvent}
   > n33 = Set Preserve Folder Structure LS(Preserve Folder Structure LS:bool=@n32.bIsChecked) -> (Output_Get:bool)
   > SetVisibility(
    self=Get TargetLSSubfolderText() -> (TargetLSSubfolderText:Text Object Reference)
    InVisibility=Select(Option 0=Collapsed, Option 1=Visible, Index=@n33)
   )
   n34 = "On Check State Changed (SubfolderStructureASCheckbox)"!() -> (OutputDelegate, then, bIsChecked:bool) {is=title K2Node_ComponentBoundEvent}
   > n35 = Set Preserve Folder Structure AS(Preserve Folder Structure AS:bool=@n34.bIsChecked) -> (Output_Get:bool)
   > SetVisibility(
    self=Get TargetASSubfolderText() -> (TargetASSubfolderText:Text Object Reference)
    InVisibility=Select(Option 0=Collapsed, Option 1=Visible, Index=@n35)
   )
   n36 = "On Check State Changed (SubfolderStructureAMCheckbox)"!() -> (OutputDelegate, then, bIsChecked:bool) {is=title K2Node_ComponentBoundEvent}
   > n37 = Set Preserve Folder Structure AM(Preserve Folder Structure AM:bool=@n36.bIsChecked) -> (Output_Get:bool)
   > SetVisibility(
    self=Get TargetAMSubfolderText() -> (TargetAMSubfolderText:Text Object Reference)
    InVisibility=Select(Option 0=Collapsed, Option 1=Visible, Index=@n37)
   )
   n38 = "On Check State Changed (SelectAllCheckbox)"!() -> (OutputDelegate, then, bIsChecked:bool) {is=title K2Node_ComponentBoundEvent}
   > n39 = For Each Loop(Array=GetDisplayedEntryWidgets(self=Get LevelSequenceListView()))
   > n40 = Cast To EUW_AnimBatchDup_LS_Item(Object=@n39)
   > SetSelected!(execute, self:EUW Anim Batch Dup LS Item Object Reference=@n40, Selected:bool=@n38.bIsChecked) -> (then)
   "On Property Changed (TargetACPropertyView)"!() -> (OutputDelegate, then, PropertyName:Name=None) {is=title K2Node_ComponentBoundEvent}
   > Hide/Show Widget(Widget=Make Array("[0]"=Get TargetPathVerticalBox() -> (TargetPathVerticalBox:Vertical Box Object Reference)), Hide=false)
   > Set Path Texts from AC(
    AC=Get Target AC()
    LevelSequence=Get TargetLevelSequencesText() -> (TargetLevelSequencesText:Text Object Reference)
    AnimSequence=Get TargetAnimSequencesText() -> (TargetAnimSequencesText:Text Object Reference)
    AnimMontage=Get TargetAnimMontagesText() -> (TargetAnimMontagesText:Text Object Reference)
   )
   n41 = Get AC()
   n42 = Get Target AC()
   n43 = NOT Boolean(A="Equal Exactly (String)"(A=GetObjectName(Object=@n41), B=GetObjectName(Object=@n42)))
   n44 = AND Boolean(A=IsValid(Object=Get AC()), B=IsValid(Object=Get Target AC()))
   > Hide/Show Widget(Widget=Make Array("[0]"=Get DuplicateButtonBox()), Hide=NOT Boolean(A=AND Boolean(A=@n44, B=@n43)))
   "On Clicked (DocumentationButton)"() {is=title K2Node_ComponentBoundEvent}
   > Execute Python Script!(execute, PythonScript="import webbrowser\r\nwebbrowser.open(url)", url:String=Get Documentation URL() -> (Documentation URL:String)) -> (then, ReturnValue) {inputs [url]}
   Event Construct()
   > Hide/Show Widget(
    Widget=Make Array!(
     "[0]"=Get SourcePathVerticalBox()
     "[1]":Widget Object Reference=Get TargetPathVerticalBox()
     "[2]":Widget Object Reference=Get DuplicateButtonBox()
     "[3]":Widget Object Reference=Get TargetAMSubfolderText()
    ) -> (Array)
    Hide=true
   )
   Event PreConstruct()
   > SetObject(self=Get SourceACPropertyView() -> (SourceACPropertyView:Single Property View Object Reference), NewObject=Self-Reference())
   > SetObject(self=Get TargetACPropertyView() -> (TargetACPropertyView:Single Property View Object Reference), NewObject=Self-Reference())
   n45 = Construct Object from Class!(
    execute
    Class=/AnimInEdTools/AnimBatchDuplicate/Data/BP_AnimBatchDup_SubItem_LS.BP_AnimBatchDup_SubItem_LS_C
    self
    Level Sequence=@n12
    Source Name:String
    Target Name:String
    Source Full Path:String
   ) -> (then, ReturnValue:BP Anim Batch Dup Sub Item LS Object Reference)
   > Construct Object from Class!(
    execute
    Class=/AnimInEdTools/AnimBatchDuplicate/Data/BP_AnimBatchDup_Item.BP_AnimBatchDup_Item_C
    self
    Level Sequence Item:BP Anim Batch Dup Sub Item LS Object Reference=@n45
   ) -> (then, ReturnValue:BP Anim Batch Dup Item Object Reference)
   Get Current List Items() -> (Current List Items:Array of BP Anim Batch Dup Item Object References)
  ]}
 Get Anim Set Name {nodes [
   Get Anim Set Name!() -> (then, AC:Animatable Creature Object Reference) {is=title K2Node_FunctionEntry}
   > n1 = Sequence!(execute) -> (then_0, then_1)
   n2 = Get AC()
   > n3 = For Each Loop with Break!(
    Exec:Exec
    Array:Array of Strings=Get Anim Set!(self:Animatable Creature Object Reference=@n2) -> (Subfolders:Array of Strings) {is=title K2Node_CallFunction}
    Break:Exec=@n4.then
   ) -> (LoopBody:Exec, Array Element:String, Array Index:int, Completed:Exec)
   > Branch(Condition=Contains(SearchIn=Get Variant!(self:Animatable Creature Object Reference=@n2) -> (Variant:String), Substring=@n3, bUseCase=true))
   > n4 = Set Local Name(Local Name:String=@n3) -> (Output_Get:String)
   Return Node!(execute=@n1.then_1, Name:String=Get Local Name() -> (Local Name:String))
  ]}
 Get Last Folder Name From Path {nodes [
   Get Last Folder Name From Path!() -> (then, Path:String) {is=title K2Node_FunctionEntry}
   > n1 = Sequence()
   > Branch(Condition=ValidatePath!(InPath:String=Get Path() -> (Path:String)) -> (bDidSucceed:bool, OutReason:Text))
   n2 = Split(SourceString=Get Path(), InStr=., SearchCase=CaseSensitive, SearchDir=FromEnd)
   > Set Local Folder(Local Folder:String=Split(
     SourceString=Split(SourceString=@n2, InStr=/, SearchCase=CaseSensitive, SearchDir=FromEnd)
     InStr=/
     SearchCase=CaseSensitive
     SearchDir=FromEnd
    ).RightS) -> (Output_Get:String)
   Return Node!(execute=@n1.then_1, Folder:String=Get Local Folder() -> (Local Folder:String))
  ]}
 Hide/Show Widget {nodes [
   n1 = Hide/Show Widget!() -> (then, Widget:Array of Widget Object References, Hide:bool=~) {is=title K2Node_FunctionEntry}
   > n2 = For Each Loop(Array:Array of Widget Object References=@n1) -> (Array Element:Widget Object Reference)
   > n3 = Branch(Condition=Get Hide() -> (Hide:bool))
   > SetVisibility(self=@n2, InVisibility=Collapsed)
   SetVisibility(execute=@n3.else, self=@n2, InVisibility=Visible)
  ]}
 Set Path Texts from AC {nodes [
   Set Path Texts from AC!() -> (
    then
    AC:Animatable Creature Object Reference
    LevelSequence:Text Object Reference
    AnimSequence:Text Object Reference
    AnimMontage:Text Object Reference
   ) {is=title K2Node_FunctionEntry}
   > n1 = Sequence!(execute) -> (then_0, then_1, then_2:Exec)
   n2 = Get AC()
   > Set Text(
    Text="To Text (String)"(InString=Translate Source Path If Tokenized(AC=@n2, Anim Folder=NewEnumerator0))
    self=Get LevelSequence() -> (LevelSequence:Text Object Reference)
   )
   n3 = Translate Source Path If Tokenized(AC=@n2, Anim Folder=NewEnumerator1)
   Set Text(execute=@n1.then_1, Text="To Text (String)"(InString=@n3), self=Get AnimSequence() -> (AnimSequence:Text Object Reference))
   n4 = Translate Source Path If Tokenized(AC=@n2, Anim Folder=NewEnumerator2)
   Set Text(
    execute=@n1.then_2
    Text=SelectText(A="To Text (String)"(InString=@n3), B="To Text (String)"(InString=@n4), bPickA="Equal Exactly (String)"(A=@n4))
    self=Get AnimMontage() -> (AnimMontage:Text Object Reference)
   )
  ]}
 Translate Source Path If Tokenized {nodes [
   Translate Source Path If Tokenized!() -> (then, AC:Animatable Creature Object Reference, Anim Folder:AnimBatchDup_Folder Enum=NewEnumerator0) {is=title K2Node_FunctionEntry}
   > n1 = Sequence()
   Return Node!(execute=@n1.then_2, Path:String=Get Final Path() -> (Final Path:String))
   > n2 = Switch on AnimBatchDup_Folder!(execute, Selection:AnimBatchDup_Folder Enum=Get Anim Folder() -> (Anim Folder:AnimBatchDup_Folder Enum=NewEnumerator0)) -> (NewEnumerator0:Exec, NewEnumerator1:Exec, NewEnumerator2:Exec) {is=title K2Node_SwitchEnum}
   > Set Initial Path(Initial Path:String=Get Source Path!(self:Animatable Creature Object Reference=Get AC()) -> (Source Path:String)) -> (Output_Get:String)
   Set Initial Path(execute=@n2.NewEnumerator1, Initial Path=Get Animation Path!(self:Animatable Creature Object Reference=Get AC()) -> (Animation Path:String))
   Set Initial Path(
    execute=@n2.NewEnumerator2
    Initial Path=Get AnimMontage Path!(self:Animatable Creature Object Reference=Get AC()) -> (AnimMontage Path:String)
   )
   n3 = Get Initial Path() -> (Initial Path:String)
   n4 = Branch(execute=@n1.then_1, Condition=OR Boolean(A=Contains(SearchIn=@n3, Substring="{"), B=Contains(SearchIn=@n3, Substring="}")))
   n5 = Get AC()
   n6 = Get AC()
   n7 = Get Naming Convention!(self:Animatable Creature Object Reference=@n6) -> (ReturnValue:EU Animation Naming Convention Object Reference) {is=title K2Node_CallFunction}
   n8 = Get Asset!(self:Animatable Creature Object Reference=@n6) -> (Asset:String)
   n9 = Get Variant(self=@n6)
   > Set Final Path(Final Path:String=Get Path!(
     self:EU Animation Naming Convention Object Reference=Get Naming Convention(self=@n5)
     Token Map:Map of Strings to Strings=Get Path Parts Map!(
      Asset:String=@n8
      Animation Set:String
      Animation Name:String
      Descriptor:String
      Direction:String
      Subfolder:String
      Variation:String
      Plugin:String=@n9
      __WorldContext:Object Reference=@n7
     ) -> (Output Path Parts Map:Map of Strings to Strings) {is=title K2Node_CallFunction}
     Path Pattern:String=Get Initial Path()
     Remove Spaces On Name:bool=true
    ) -> (File Path:String, Folder Path:String, File Name:String) {is=title K2Node_CallFunction}.Folder Path) -> (Output_Get:String)
   Set Final Path(execute=@n4.else, Final Path=Get Initial Path())
  ]}
]
variables [
 AC {type=AnimatableCreature_C, default=null}
 AnimPaths {
  type="array<string>"
  default [
   /JunoEditor/AnimBatchDuplicate/Output_Test/_07/AS/Minifig_ThreeHandedHammer_ChargeAttack_AnimSET
   /JunoEditor/AnimBatchDuplicate/Output_Test/_07/AS/Minifig_ThreeHandedHammer_ChargeAttack_Intro
   /JunoEditor/AnimBatchDuplicate/Output_Test/_07/AS/Minifig_ThreeHandedHammer_ChargeAttack_Loop
   /JunoEditor/AnimBatchDuplicate/Output_Test/_07/AS/Minifig_ThreeHandedHammer_ChargeAttack_Release
   /JunoEditor/AnimBatchDuplicate/Output_Test/_07/AS/Minifig_ThreeHandedHammer_ChargeAttack_Secondary
  ]
 }
 Current List Items {type="array<BP_AnimBatchDup_Item_C>", default []}
 Documentation URL {type=string, default=https://epicgames.atlassian.net/wiki/spaces/PP/pages/281937101/Animation+Batch+Duplicate+tool}
 Preserve Folder Structure AM {type=boolean}
 Preserve Folder Structure AS {type=boolean, default=true}
 Preserve Folder Structure LS {type=boolean, default=true}
 Source AM Replaceable {type=string, default=""}
 Source AS Replaceable {type=string, default=""}
 Source LS Replaceable {type=string, default=""}
 Source Set Name {type=string, default=""}
 Target AC {type=AnimatableCreature_C, default=null}
 Target AM Replaceable {type=string, default=""}
 Target AS Replaceable {type=string, default=""}
 Target LS Replaceable {type=string, default=""}
]
widgets {
 info {rootWidgetClass=&CanvasPanel, widgetCount=87}
 widgets [
  CanvasPanel_0 {widgetClassPath=&CanvasPanel, widget=CanvasPanel_0, properties {bOverride_Cursor=true}}
  VerticalBox_94 {
   widgetClassPath=&VerticalBox
   parent=CanvasPanel_0
   widget=VerticalBox_94
   slot {class=CanvasPanelSlot, properties {layoutData {alignment=<0 0>, anchors {maximum=<1 1>, minimum=<0 0>}, offsets=<ltrb 0 0 0 0>}}}
  }
  InputsBox {
   widgetClassPath=&HorizontalBox
   parent=VerticalBox_94
   widget=InputsBox
   slot {class=VerticalBoxSlot, properties {padding=<ltrb 38 0 0 0>}}
  }
  VerticalBox_0 {
   widgetClassPath=&VerticalBox
   parent=InputsBox
   widget=VerticalBox_0
   slot {class=HorizontalBoxSlot, properties {size {sizeRule=Fill, value=1}}}
  }
  SourceVerticalBox {
   widgetClassPath=&VerticalBox
   parent=VerticalBox_0
   widget=SourceVerticalBox
   slot {class=VerticalBoxSlot, properties {size {sizeRule=Fill, value=1}}}
  }
  Border_91 {
   widgetClassPath=&Border
   parent=SourceVerticalBox
   widget=Border_91
   properties {
    background {
     bIsDynamicallyLoaded=false
     drawAs=Image
     imageType=NoImage
     mirroring=NoMirror
     resourceName=None
     resourceObject=null
     tiling=NoTile
     imageSize=<32 32>
     margin=<ltrb 0 0 0 0>
     outlineSettings {
      bUseBrushTransparency=false
      roundingType=HalfHeightRadius
      width=0
      color {colorUseRule=UseColor_Specified, specifiedColor=<rgba 0 0 0 0>}
      cornerRadii=<0 0 0 0>
     }
     tintColor {colorUseRule=UseColor_Specified, specifiedColor=<rgba 0.0176 0.0176 0.0176 1>}
     uVRegion {bIsValid=false, max=<0 0>, min=<0 0>}
    }
    padding=<ltrb 10 10 10 10>
   }
   slot {class=VerticalBoxSlot}
  }
  SourceACPropertyView {
   widgetClassPath=&SinglePropertyView
   parent=Border_91
   widget=SourceACPropertyView
   properties {nameOverride=Source AC, propertyName=AC}
   slot {class=BorderSlot, properties {padding=<ltrb 10 10 10 10>}}
  }
  Border_93 {
   widgetClassPath=&Border
   parent=SourceVerticalBox
   widget=Border_93
   properties {
    background {
     bIsDynamicallyLoaded=false
     drawAs=Image
     imageType=NoImage
     mirroring=NoMirror
     resourceName=None
     resourceObject=null
     tiling=NoTile
     imageSize=<32 32>
     margin=<ltrb 0 0 0 0>
     outlineSettings {
      bUseBrushTransparency=false
      roundingType=HalfHeightRadius
      width=0
      color {colorUseRule=UseColor_Specified, specifiedColor=<rgba 0 0 0 0>}
      cornerRadii=<0 0 0 0>
     }
     tintColor {colorUseRule=UseColor_Specified, specifiedColor=<rgba 0.0176 0.0176 0.0176 1>}
     uVRegion {bIsValid=false, max=<0 0>, min=<0 0>}
    }
    padding=<ltrb 10 0 10 15>
   }
   slot {class=VerticalBoxSlot, properties {size {sizeRule=Fill, value=1}}}
  }
  SourcePathVerticalBox {
   widgetClassPath=&VerticalBox
   bIsVariable=true
   parent=Border_93
   widget=SourcePathVerticalBox
   slot {class=BorderSlot, properties {padding=<ltrb 10 0 10 15>}}
  }
  VerticalBox_5 {
   widgetClassPath=&VerticalBox
   parent=SourcePathVerticalBox
   widget=VerticalBox_5
   slot {class=VerticalBoxSlot, properties {padding=<ltrb 0 0 0 5>}}
  }
  TextBlock_140 {
   widgetClassPath=&TextBlock
   parent=VerticalBox_5
   widget=TextBlock_140
   properties {
    text=Source Animations
    font {
     bForceMonospaced=false
     bMaterialIsStencil=false
     fontMaterial=null
     fontObject=&Roboto
     letterSpacing=0
     monospacedWidth=1
     size=9
     skewAmount=0
     typefaceFontName=Bold
     outlineSettings {
      bApplyOutlineToDropShadows=false
      bMiteredCorners=false
      bSeparateFillAlpha=false
      outlineMaterial=null
      outlineSize=0
      outlineColor=<rgba 0 0 0 1>
     }
    }
   }
   slot {class=VerticalBoxSlot, properties {padding=<ltrb 0 0 0 5>}}
  }
  HorizontalBox_1 {
   widgetClassPath=&HorizontalBox
   parent=VerticalBox_5
   widget=HorizontalBox_1
   slot {class=VerticalBoxSlot, properties {padding=<ltrb 0 0 0 5>}}
  }
  TextBlock_67 {
   widgetClassPath=&TextBlock
   parent=HorizontalBox_1
   widget=TextBlock_67
   properties {
    text=LevelSequences gathered from
    colorAndOpacity {colorUseRule=UseColor_Specified, specifiedColor=<rgba 0.5776 0.0802 0.0802 1>}
    font {
     bForceMonospaced=false
     bMaterialIsStencil=false
     fontMaterial=null
     fontObject=&Roboto
     letterSpacing=0
     monospacedWidth=1
     size=9
     skewAmount=0
     typefaceFontName=Bold
     outlineSettings {
      bApplyOutlineToDropShadows=false
      bMiteredCorners=false
      bSeparateFillAlpha=false
      outlineMaterial=null
      outlineSize=0
      outlineColor=<rgba 0 0 0 1>
     }
    }
   }
   slot {class=HorizontalBoxSlot}
  }
  SourceLevelSequencesText {
   widgetClassPath=&TextBlock
   bIsVariable=true
   parent=HorizontalBox_1
   widget=SourceLevelSequencesText
   properties {font {
     bForceMonospaced=false
     bMaterialIsStencil=false
     fontMaterial=null
     fontObject=&Roboto
     letterSpacing=0
     monospacedWidth=1
     size=9
     skewAmount=0
     typefaceFontName=Italic
     outlineSettings {
      bApplyOutlineToDropShadows=false
      bMiteredCorners=false
      bSeparateFillAlpha=false
      outlineMaterial=null
      outlineSize=0
      outlineColor=<rgba 0 0 0 1>
     }
    }}
   slot {class=HorizontalBoxSlot, properties {padding=<ltrb 5 0 0 0>}}
  }
  HorizontalBox {
   widgetClassPath=&HorizontalBox
   parent=VerticalBox_5
   widget=HorizontalBox
   slot {class=VerticalBoxSlot, properties {padding=<ltrb 0 0 0 5>}}
  }
  TextBlock {
   widgetClassPath=&TextBlock
   parent=HorizontalBox
   widget=TextBlock
   properties {
    text=AnimSequences are gathered from LevelSequenceAnimSequenceLinks
    colorAndOpacity {colorUseRule=UseColor_Specified, specifiedColor=<rgba 0.0802 0.1981 0.0648 1>}
    font {
     bForceMonospaced=false
     bMaterialIsStencil=false
     fontMaterial=null
     fontObject=&Roboto
     letterSpacing=0
     monospacedWidth=1
     size=9
     skewAmount=0
     typefaceFontName=Bold
     outlineSettings {
      bApplyOutlineToDropShadows=false
      bMiteredCorners=false
      bSeparateFillAlpha=false
      outlineMaterial=null
      outlineSize=0
      outlineColor=<rgba 0 0 0 1>
     }
    }
   }
   slot {class=HorizontalBoxSlot}
  }
  HorizontalBox_3 {
   widgetClassPath=&HorizontalBox
   parent=VerticalBox_5
   widget=HorizontalBox_3
   properties {visibility=Visible}
   slot {class=VerticalBoxSlot, properties {padding=<ltrb 0 0 0 5>}}
  }
  TextBlock_2 {
   widgetClassPath=&TextBlock
   parent=HorizontalBox_3
   widget=TextBlock_2
   properties {
    text=AnimMontages are gathered from AnimSequences referencers
    colorAndOpacity {colorUseRule=UseColor_Specified, specifiedColor=<rgba 0.1274 0.1274 1 1>}
    font {
     bForceMonospaced=false
     bMaterialIsStencil=false
     fontMaterial=null
     fontObject=&Roboto
     letterSpacing=0
     monospacedWidth=1
     size=9
     skewAmount=0
     typefaceFontName=Bold
     outlineSettings {
      bApplyOutlineToDropShadows=false
      bMiteredCorners=false
      bSeparateFillAlpha=false
      outlineMaterial=null
      outlineSize=0
      outlineColor=<rgba 0 0 0 1>
     }
    }
   }
   slot {class=HorizontalBoxSlot}
  }
  VerticalBox_6 {
   widgetClassPath=&VerticalBox
   parent=SourcePathVerticalBox
   widget=VerticalBox_6
   slot {class=VerticalBoxSlot, properties {verticalAlignment=VAlign_Bottom, padding=<ltrb 0 10 0 0>, size {sizeRule=Fill, value=1}}}
  }
  TextBlock_7 {
   widgetClassPath=&TextBlock
   parent=VerticalBox_6
   widget=TextBlock_7
   properties {
    text=Source Strings to Replace
    font {
     bForceMonospaced=false
     bMaterialIsStencil=false
     fontMaterial=null
     fontObject=&Roboto
     letterSpacing=0
     monospacedWidth=1
     size=9
     skewAmount=0
     typefaceFontName=Bold
     outlineSettings {
      bApplyOutlineToDropShadows=false
      bMiteredCorners=false
      bSeparateFillAlpha=false
      outlineMaterial=null
      outlineSize=0
      outlineColor=<rgba 0 0 0 1>
     }
    }
   }
   slot {class=VerticalBoxSlot, properties {padding=<ltrb 0 10 0 5>}}
  }
  HorizontalBox_17 {widgetClassPath=&HorizontalBox, parent=VerticalBox_6, widget=HorizontalBox_17, slot {class=VerticalBoxSlot}}
  VerticalBox_8 {
   widgetClassPath=&VerticalBox
   parent=HorizontalBox_17
   widget=VerticalBox_8
   slot {class=HorizontalBoxSlot, properties {size {sizeRule=Fill, value=1}}}
  }
  HorizontalBox_9 {widgetClassPath=&HorizontalBox, parent=VerticalBox_8, widget=HorizontalBox_9, slot {class=VerticalBoxSlot}}
  TextBlock_8 {
   widgetClassPath=&TextBlock
   parent=HorizontalBox_9
   widget=TextBlock_8
   properties {
    text=LevelSequences
    colorAndOpacity {colorUseRule=UseColor_Specified, specifiedColor=<rgba 0.5776 0.0802 0.0802 1>}
    font {
     bForceMonospaced=false
     bMaterialIsStencil=false
     fontMaterial=null
     fontObject=&Roboto
     letterSpacing=0
     monospacedWidth=1
     size=9
     skewAmount=0
     typefaceFontName=Bold
     outlineSettings {
      bApplyOutlineToDropShadows=false
      bMiteredCorners=false
      bSeparateFillAlpha=false
      outlineMaterial=null
      outlineSize=0
      outlineColor=<rgba 0 0 0 1>
     }
    }
   }
   slot {class=HorizontalBoxSlot, properties {verticalAlignment=VAlign_Center}}
  }
  SourceLSTextBox {
   widgetClassPath=&EditorUtilityEditableTextBox
   bIsVariable=true
   parent=HorizontalBox_9
   widget=SourceLSTextBox
   slot {class=HorizontalBoxSlot, properties {padding=<ltrb 10 0 10 0>, size {sizeRule=Fill, value=1}}}
  }
  HorizontalBox_11 {widgetClassPath=&HorizontalBox, parent=VerticalBox_8, widget=HorizontalBox_11, slot {class=VerticalBoxSlot}}
  TextBlock_9 {
   widgetClassPath=&TextBlock
   parent=HorizontalBox_11
   widget=TextBlock_9
   properties {
    text=AnimSequences
    colorAndOpacity {colorUseRule=UseColor_Specified, specifiedColor=<rgba 0.0802 0.1981 0.0648 1>}
    font {
     bForceMonospaced=false
     bMaterialIsStencil=false
     fontMaterial=null
     fontObject=&Roboto
     letterSpacing=0
     monospacedWidth=1
     size=9
     skewAmount=0
     typefaceFontName=Bold
     outlineSettings {
      bApplyOutlineToDropShadows=false
      bMiteredCorners=false
      bSeparateFillAlpha=false
      outlineMaterial=null
      outlineSize=0
      outlineColor=<rgba 0 0 0 1>
     }
    }
   }
   slot {class=HorizontalBoxSlot, properties {verticalAlignment=VAlign_Center}}
  }
  SourceASTextBox {
   widgetClassPath=&EditorUtilityEditableTextBox
   bIsVariable=true
   parent=HorizontalBox_11
   widget=SourceASTextBox
   slot {class=HorizontalBoxSlot, properties {padding=<ltrb 10 0 10 0>, size {sizeRule=Fill, value=1}}}
  }
  HorizontalBox_12 {
   widgetClassPath=&HorizontalBox
   parent=VerticalBox_8
   widget=HorizontalBox_12
   properties {visibility=Visible}
   slot {class=VerticalBoxSlot}
  }
  TextBlock_10 {
   widgetClassPath=&TextBlock
   parent=HorizontalBox_12
   widget=TextBlock_10
   properties {
    text=AnimMontages
    colorAndOpacity {colorUseRule=UseColor_Specified, specifiedColor=<rgba 0.1274 0.1274 1 1>}
    font {
     bForceMonospaced=false
     bMaterialIsStencil=false
     fontMaterial=null
     fontObject=&Roboto
     letterSpacing=0
     monospacedWidth=1
     size=9
     skewAmount=0
     typefaceFontName=Bold
     outlineSettings {
      bApplyOutlineToDropShadows=false
      bMiteredCorners=false
      bSeparateFillAlpha=false
      outlineMaterial=null
      outlineSize=0
      outlineColor=<rgba 0 0 0 1>
     }
    }
   }
   slot {class=HorizontalBoxSlot, properties {verticalAlignment=VAlign_Center}}
  }
  SourceAMTextBox {
   widgetClassPath=&EditorUtilityEditableTextBox
   bIsVariable=true
   parent=HorizontalBox_12
   widget=SourceAMTextBox
   slot {class=HorizontalBoxSlot, properties {padding=<ltrb 10 0 10 0>, size {sizeRule=Fill, value=1}}}
  }
  TargetVerticalBox {
   widgetClassPath=&VerticalBox
   parent=InputsBox
   widget=TargetVerticalBox
   slot {class=HorizontalBoxSlot, properties {size {sizeRule=Fill, value=1}}}
  }
  Border_94 {
   widgetClassPath=&Border
   parent=TargetVerticalBox
   widget=Border_94
   properties {
    background {
     bIsDynamicallyLoaded=false
     drawAs=Image
     imageType=NoImage
     mirroring=NoMirror
     resourceName=None
     resourceObject=null
     tiling=NoTile
     imageSize=<32 32>
     margin=<ltrb 0 0 0 0>
     outlineSettings {
      bUseBrushTransparency=false
      roundingType=HalfHeightRadius
      width=0
      color {colorUseRule=UseColor_Specified, specifiedColor=<rgba 0 0 0 0>}
      cornerRadii=<0 0 0 0>
     }
     tintColor {colorUseRule=UseColor_Specified, specifiedColor=<rgba 0.0176 0.0176 0.0176 1>}
     uVRegion {bIsValid=false, max=<0 0>, min=<0 0>}
    }
    padding=<ltrb 25 10 10 10>
   }
   slot {class=VerticalBoxSlot}
  }
  TargetACHelpButtonBox {
   widgetClassPath=&HorizontalBox
   parent=Border_94
   widget=TargetACHelpButtonBox
   slot {class=BorderSlot, properties {padding=<ltrb 25 10 10 10>}}
  }
  TargetACPropertyView {
   widgetClassPath=&SinglePropertyView
   parent=TargetACHelpButtonBox
   widget=TargetACPropertyView
   properties {propertyName=Target AC}
   slot {class=HorizontalBoxSlot, properties {size {sizeRule=Fill, value=1}}}
  }
  DocumentationButton {
   widgetClassPath=&EditorUtilityButton
   parent=TargetACHelpButtonBox
   widget=DocumentationButton
   properties {toolTipText=Documentation site}
   slot {class=HorizontalBoxSlot}
  }
  TextBlock_98 {
   widgetClassPath=&TextBlock
   parent=DocumentationButton
   widget=TextBlock_98
   properties {text=?}
   slot {class=ButtonSlot, properties {padding=<ltrb 0 0 0 0>}}
  }
  Border_95 {
   widgetClassPath=&Border
   parent=TargetVerticalBox
   widget=Border_95
   properties {
    background {
     bIsDynamicallyLoaded=false
     drawAs=Image
     imageType=NoImage
     mirroring=NoMirror
     resourceName=None
     resourceObject=null
     tiling=NoTile
     imageSize=<32 32>
     margin=<ltrb 0 0 0 0>
     outlineSettings {
      bUseBrushTransparency=false
      roundingType=HalfHeightRadius
      width=0
      color {colorUseRule=UseColor_Specified, specifiedColor=<rgba 0 0 0 0>}
      cornerRadii=<0 0 0 0>
     }
     tintColor {colorUseRule=UseColor_Specified, specifiedColor=<rgba 0.0176 0.0176 0.0176 1>}
     uVRegion {bIsValid=false, max=<0 0>, min=<0 0>}
    }
    padding=<ltrb 25 0 10 15>
   }
   slot {class=VerticalBoxSlot, properties {size {sizeRule=Fill, value=1}}}
  }
  TargetPathVerticalBox {
   widgetClassPath=&VerticalBox
   bIsVariable=true
   parent=Border_95
   widget=TargetPathVerticalBox
   slot {class=BorderSlot, properties {padding=<ltrb 25 0 10 15>}}
  }
  TextBlock_6 {
   widgetClassPath=&TextBlock
   parent=TargetPathVerticalBox
   widget=TextBlock_6
   properties {
    text="Target Paths (from Target AC)"
    font {
     bForceMonospaced=false
     bMaterialIsStencil=false
     fontMaterial=null
     fontObject=&Roboto
     letterSpacing=0
     monospacedWidth=1
     size=9
     skewAmount=0
     typefaceFontName=Bold
     outlineSettings {
      bApplyOutlineToDropShadows=false
      bMiteredCorners=false
      bSeparateFillAlpha=false
      outlineMaterial=null
      outlineSize=0
      outlineColor=<rgba 0 0 0 1>
     }
    }
   }
   slot {class=VerticalBoxSlot, properties {padding=<ltrb 0 0 0 5>}}
  }
  HorizontalBox_8 {
   widgetClassPath=&HorizontalBox
   parent=TargetPathVerticalBox
   widget=HorizontalBox_8
   properties {bOverride_Cursor=true}
   slot {class=VerticalBoxSlot, properties {padding=<ltrb 0 0 0 2>}}
  }
  HorizontalBox_0 {
   widgetClassPath=&HorizontalBox
   parent=HorizontalBox_8
   widget=HorizontalBox_0
   slot {class=HorizontalBoxSlot, properties {padding=<ltrb 0 0 10 0>}}
  }
  TextBlock_198 {
   widgetClassPath=&TextBlock
   parent=HorizontalBox_0
   widget=TextBlock_198
   properties {
    text=Subfolder?
    colorAndOpacity {colorUseRule=UseColor_Specified, specifiedColor=<rgba 0.4678 0.2747 0.0908 1>}
    font {
     bForceMonospaced=false
     bMaterialIsStencil=false
     fontMaterial=null
     fontObject=&Roboto
     letterSpacing=0
     monospacedWidth=1
     size=9
     skewAmount=0
     typefaceFontName=Bold
     outlineSettings {
      bApplyOutlineToDropShadows=false
      bMiteredCorners=false
      bSeparateFillAlpha=false
      outlineMaterial=null
      outlineSize=0
      outlineColor=<rgba 0 0 0 1>
     }
    }
   }
   slot {class=HorizontalBoxSlot, properties {verticalAlignment=VAlign_Center, padding=<ltrb 0 0 5 0>}}
  }
  SubfolderStructureLSCheckbox {
   widgetClassPath=&EditorUtilityCheckBox
   bIsVariable=true
   parent=HorizontalBox_0
   widget=SubfolderStructureLSCheckbox
   properties {
    checkedState=Checked
    horizontalAlignment=HAlign_Center
    widgetStyle {
     checkBoxType=CheckBox
     backgroundHoveredImage {
      bIsDynamicallyLoaded=false
      drawAs=RoundedBox
      imageType=NoImage
      mirroring=NoMirror
      resourceName=None
      resourceObject=null
      tiling=NoTile
      imageSize=<18 18>
      margin=<ltrb 0 0 0 0>
      outlineSettings {
       bUseBrushTransparency=false
       roundingType=FixedRadius
       width=1
       color {colorUseRule=UseColor_ColorTable, specifiedColor=<rgba 0 0 0 0>}
       cornerRadii=<3 3 3 3>
      }
      tintColor {colorUseRule=UseColor_Specified, specifiedColor=<rgba 0.0048 0.0048 0.0048 1>}
      uVRegion {bIsValid=false, max=<0 0>, min=<0 0>}
     }
     backgroundImage {
      bIsDynamicallyLoaded=false
      drawAs=RoundedBox
      imageType=NoImage
      mirroring=NoMirror
      resourceName=None
      resourceObject=null
      tiling=NoTile
      imageSize=<18 18>
      margin=<ltrb 0 0 0 0>
      outlineSettings {
       bUseBrushTransparency=false
       roundingType=FixedRadius
       width=1
       color {colorUseRule=UseColor_ColorTable, specifiedColor=<rgba 0 0 0 0>}
       cornerRadii=<3 3 3 3>
      }
      tintColor {colorUseRule=UseColor_Specified, specifiedColor=<rgba 0.0048 0.0048 0.0048 1>}
      uVRegion {bIsValid=false, max=<0 0>, min=<0 0>}
     }
     backgroundPressedImage {
      bIsDynamicallyLoaded=false
      drawAs=RoundedBox
      imageType=NoImage
      mirroring=NoMirror
      resourceName=None
      resourceObject=null
      tiling=NoTile
      imageSize=<18 18>
      margin=<ltrb 0 0 0 0>
      outlineSettings {
       bUseBrushTransparency=false
       roundingType=FixedRadius
       width=1
       color {colorUseRule=UseColor_ColorTable, specifiedColor=<rgba 0 0 0 0>}
       cornerRadii=<3 3 3 3>
      }
      tintColor {colorUseRule=UseColor_Specified, specifiedColor=<rgba 0.0048 0.0048 0.0048 1>}
      uVRegion {bIsValid=false, max=<0 0>, min=<0 0>}
     }
     borderBackgroundColor {colorUseRule=UseColor_Specified, specifiedColor=<rgba 1 1 1 1>}
     checkedForeground {colorUseRule=UseColor_Specified, specifiedColor=<rgba 1 1 1 1>}
     checkedHoveredForeground {colorUseRule=UseColor_Specified, specifiedColor=<rgba 1 1 1 1>}
     checkedHoveredImage {
      bIsDynamicallyLoaded=false
      drawAs=Image
      imageType=Vector
      mirroring=NoMirror
      resourceName=../../../Engine/Content/Slate/Starship/CoreWidgets/CheckBox/check.svg
      resourceObject=null
      tiling=NoTile
      imageSize=<16 16>
      margin=<ltrb 0 0 0 0>
      outlineSettings {
       bUseBrushTransparency=false
       roundingType=HalfHeightRadius
       width=0
       color {colorUseRule=UseColor_Specified, specifiedColor=<rgba 0 0 0 0>}
       cornerRadii=<0 0 0 0>
      }
      tintColor {colorUseRule=UseColor_Specified, specifiedColor=<rgba 0.0044 0.2384 1 1>}
      uVRegion {bIsValid=false, max=<0 0>, min=<0 0>}
     }
     checkedImage {
      bIsDynamicallyLoaded=false
      drawAs=Image
      imageType=Vector
      mirroring=NoMirror
      resourceName=../../../Engine/Content/Slate/Starship/CoreWidgets/CheckBox/check.svg
      resourceObject=null
      tiling=NoTile
      imageSize=<16 16>
      margin=<ltrb 0 0 0 0>
      outlineSettings {
       bUseBrushTransparency=false
       roundingType=HalfHeightRadius
       width=0
       color {colorUseRule=UseColor_Specified, specifiedColor=<rgba 0 0 0 0>}
       cornerRadii=<0 0 0 0>
      }
      tintColor {colorUseRule=UseColor_Specified, specifiedColor=<rgba 0 0.162 0.7454 1>}
      uVRegion {bIsValid=false, max=<0 0>, min=<0 0>}
     }
     checkedPressedForeground {colorUseRule=UseColor_Specified, specifiedColor=<rgba 1 1 1 1>}
     checkedPressedImage {
      bIsDynamicallyLoaded=false
      drawAs=Image
      imageType=Vector
      mirroring=NoMirror
      resourceName=../../../Engine/Content/Slate/Starship/CoreWidgets/CheckBox/check.svg
      resourceObject=null
      tiling=NoTile
      imageSize=<16 16>
      margin=<ltrb 0 0 0 0>
      outlineSettings {
       bUseBrushTransparency=false
       roundingType=HalfHeightRadius
       width=0
       color {colorUseRule=UseColor_Specified, specifiedColor=<rgba 0 0 0 0>}
       cornerRadii=<0 0 0 0>
      }
      tintColor {colorUseRule=UseColor_Specified, specifiedColor=<rgba 0 0.0802 0.3515 1>}
      uVRegion {bIsValid=false, max=<0 0>, min=<0 0>}
     }
     checkedSlateSound {resourceObject=null}
     foregroundColor {colorUseRule=UseColor_Specified, specifiedColor=<rgba 1 1 1 1>}
     hoveredForeground {colorUseRule=UseColor_Specified, specifiedColor=<rgba 1 1 1 1>}
     hoveredSlateSound {resourceObject=null}
     padding=<ltrb 0 0 0 0>
     pressedForeground {colorUseRule=UseColor_Specified, specifiedColor=<rgba 1 1 1 1>}
     uncheckedHoveredImage {
      bIsDynamicallyLoaded=false
      drawAs=NoDrawType
      imageType=NoImage
      mirroring=NoMirror
      resourceName=None
      resourceObject=null
      tiling=NoTile
      imageSize=<0 0>
      margin=<ltrb 0 0 0 0>
      outlineSettings {
       bUseBrushTransparency=false
       roundingType=HalfHeightRadius
       width=0
       color {colorUseRule=UseColor_Specified, specifiedColor=<rgba 0 0 0 0>}
       cornerRadii=<0 0 0 0>
      }
      tintColor {colorUseRule=UseColor_Specified, specifiedColor=<rgba 1 1 1 1>}
      uVRegion {bIsValid=false, max=<0 0>, min=<0 0>}
     }
     uncheckedImage {
      bIsDynamicallyLoaded=false
      drawAs=NoDrawType
      imageType=NoImage
      mirroring=NoMirror
      resourceName=None
      resourceObject=null
      tiling=NoTile
      imageSize=<0 0>
      margin=<ltrb 0 0 0 0>
      outlineSettings {
       bUseBrushTransparency=false
       roundingType=HalfHeightRadius
       width=0
       color {colorUseRule=UseColor_Specified, specifiedColor=<rgba 0 0 0 0>}
       cornerRadii=<0 0 0 0>
      }
      tintColor {colorUseRule=UseColor_Specified, specifiedColor=<rgba 1 1 1 1>}
      uVRegion {bIsValid=false, max=<0 0>, min=<0 0>}
     }
     uncheckedPressedImage {
      bIsDynamicallyLoaded=false
      drawAs=NoDrawType
      imageType=NoImage
      mirroring=NoMirror
      resourceName=None
      resourceObject=null
      tiling=NoTile
      imageSize=<0 0>
      margin=<ltrb 0 0 0 0>
      outlineSettings {
       bUseBrushTransparency=false
       roundingType=HalfHeightRadius
       width=0
       color {colorUseRule=UseColor_Specified, specifiedColor=<rgba 0 0 0 0>}
       cornerRadii=<0 0 0 0>
      }
      tintColor {colorUseRule=UseColor_Specified, specifiedColor=<rgba 1 1 1 1>}
      uVRegion {bIsValid=false, max=<0 0>, min=<0 0>}
     }
     uncheckedSlateSound {resourceObject=null}
     undeterminedForeground {colorUseRule=UseColor_Specified, specifiedColor=<rgba 1 1 1 1>}
     undeterminedHoveredImage {
      bIsDynamicallyLoaded=false
      drawAs=Image
      imageType=Vector
      mirroring=NoMirror
      resourceName=../../../Engine/Content/Slate/Starship/CoreWidgets/CheckBox/indeterminate.svg
      resourceObject=null
      tiling=NoTile
      imageSize=<16 16>
      margin=<ltrb 0 0 0 0>
      outlineSettings {
       bUseBrushTransparency=false
       roundingType=HalfHeightRadius
       width=0
       color {colorUseRule=UseColor_Specified, specifiedColor=<rgba 0 0 0 0>}
       cornerRadii=<0 0 0 0>
      }
      tintColor {colorUseRule=UseColor_Specified, specifiedColor=<rgba 0.0044 0.2384 1 1>}
      uVRegion {bIsValid=false, max=<0 0>, min=<0 0>}
     }
     undeterminedImage {
      bIsDynamicallyLoaded=false
      drawAs=Image
      imageType=Vector
      mirroring=NoMirror
      resourceName=../../../Engine/Content/Slate/Starship/CoreWidgets/CheckBox/indeterminate.svg
      resourceObject=null
      tiling=NoTile
      imageSize=<16 16>
      margin=<ltrb 0 0 0 0>
      outlineSettings {
       bUseBrushTransparency=false
       roundingType=HalfHeightRadius
       width=0
       color {colorUseRule=UseColor_Specified, specifiedColor=<rgba 0 0 0 0>}
       cornerRadii=<0 0 0 0>
      }
      tintColor {colorUseRule=UseColor_Specified, specifiedColor=<rgba 0 0.162 0.7454 1>}
      uVRegion {bIsValid=false, max=<0 0>, min=<0 0>}
     }
     undeterminedPressedImage {
      bIsDynamicallyLoaded=false
      drawAs=Image
      imageType=Vector
      mirroring=NoMirror
      resourceName=../../../Engine/Content/Slate/Starship/CoreWidgets/CheckBox/indeterminate.svg
      resourceObject=null
      tiling=NoTile
      imageSize=<16 16>
      margin=<ltrb 0 0 0 0>
      outlineSettings {
       bUseBrushTransparency=false
       roundingType=HalfHeightRadius
       width=0
       color {colorUseRule=UseColor_Specified, specifiedColor=<rgba 0 0 0 0>}
       cornerRadii=<0 0 0 0>
      }
      tintColor {colorUseRule=UseColor_Specified, specifiedColor=<rgba 0 0.0802 0.3515 1>}
      uVRegion {bIsValid=false, max=<0 0>, min=<0 0>}
     }
    }
   }
   slot {class=HorizontalBoxSlot}
  }
  HorizontalBox_2 {widgetClassPath=&HorizontalBox, parent=HorizontalBox_8, widget=HorizontalBox_2, slot {class=HorizontalBoxSlot}}
  TextBlock_1 {
   widgetClassPath=&TextBlock
   parent=HorizontalBox_2
   widget=TextBlock_1
   properties {
    text=LevelSequences will be duplicated in
    colorAndOpacity {colorUseRule=UseColor_Specified, specifiedColor=<rgba 0.5776 0.0802 0.0802 1>}
    font {
     bForceMonospaced=false
     bMaterialIsStencil=false
     fontMaterial=null
     fontObject=&Roboto
     letterSpacing=0
     monospacedWidth=1
     size=9
     skewAmount=0
     typefaceFontName=Bold
     outlineSettings {
      bApplyOutlineToDropShadows=false
      bMiteredCorners=false
      bSeparateFillAlpha=false
      outlineMaterial=null
      outlineSize=0
      outlineColor=<rgba 0 0 0 1>
     }
    }
   }
   slot {class=HorizontalBoxSlot, properties {verticalAlignment=VAlign_Center}}
  }
  TargetLevelSequencesText {
   widgetClassPath=&TextBlock
   bIsVariable=true
   parent=HorizontalBox_2
   widget=TargetLevelSequencesText
   properties {font {
     bForceMonospaced=false
     bMaterialIsStencil=false
     fontMaterial=null
     fontObject=&Roboto
     letterSpacing=0
     monospacedWidth=1
     size=9
     skewAmount=0
     typefaceFontName=Italic
     outlineSettings {
      bApplyOutlineToDropShadows=false
      bMiteredCorners=false
      bSeparateFillAlpha=false
      outlineMaterial=null
      outlineSize=0
      outlineColor=<rgba 0 0 0 1>
     }
    }}
   slot {class=HorizontalBoxSlot, properties {padding=<ltrb 5 0 0 0>}}
  }
  TargetLSSubfolderText {
   widgetClassPath=&TextBlock
   bIsVariable=true
   parent=HorizontalBox_2
   widget=TargetLSSubfolderText
   properties {
    text=/Subfolder
    colorAndOpacity {colorUseRule=UseColor_Specified, specifiedColor=<rgba 0.4678 0.2747 0.0908 1>}
    font {
     bForceMonospaced=false
     bMaterialIsStencil=false
     fontMaterial=null
     fontObject=&Roboto
     letterSpacing=0
     monospacedWidth=1
     size=9
     skewAmount=0
     typefaceFontName=Bold Italic
     outlineSettings {
      bApplyOutlineToDropShadows=false
      bMiteredCorners=false
      bSeparateFillAlpha=false
      outlineMaterial=null
      outlineSize=0
      outlineColor=<rgba 0 0 0 1>
     }
    }
   }
   slot {class=HorizontalBoxSlot, properties {verticalAlignment=VAlign_Center}}
  }
  HorizontalBox_10 {
   widgetClassPath=&HorizontalBox
   parent=TargetPathVerticalBox
   widget=HorizontalBox_10
   properties {bOverride_Cursor=true}
   slot {class=VerticalBoxSlot, properties {padding=<ltrb 0 0 0 2>}}
  }
  HorizontalBox_6 {
   widgetClassPath=&HorizontalBox
   parent=HorizontalBox_10
   widget=HorizontalBox_6
   slot {class=HorizontalBoxSlot, properties {padding=<ltrb 0 0 10 0>}}
  }
  TextBlock_5 {
   widgetClassPath=&TextBlock
   parent=HorizontalBox_6
   widget=TextBlock_5
   properties {
    text=Subfolder?
    colorAndOpacity {colorUseRule=UseColor_Specified, specifiedColor=<rgba 0.4678 0.2747 0.0908 1>}
    font {
     bForceMonospaced=false
     bMaterialIsStencil=false
     fontMaterial=null
     fontObject=&Roboto
     letterSpacing=0
     monospacedWidth=1
     size=9
     skewAmount=0
     typefaceFontName=Bold
     outlineSettings {
      bApplyOutlineToDropShadows=false
      bMiteredCorners=false
      bSeparateFillAlpha=false
      outlineMaterial=null
      outlineSize=0
      outlineColor=<rgba 0 0 0 1>
     }
    }
   }
   slot {class=HorizontalBoxSlot, properties {verticalAlignment=VAlign_Center, padding=<ltrb 0 0 5 0>}}
  }
  SubfolderStructureASCheckbox {
   widgetClassPath=&EditorUtilityCheckBox
   bIsVariable=true
   parent=HorizontalBox_6
   widget=SubfolderStructureASCheckbox
   properties {
    checkedState=Checked
    horizontalAlignment=HAlign_Center
    widgetStyle {
     checkBoxType=CheckBox
     backgroundHoveredImage {
      bIsDynamicallyLoaded=false
      drawAs=RoundedBox
      imageType=NoImage
      mirroring=NoMirror
      resourceName=None
      resourceObject=null
      tiling=NoTile
      imageSize=<18 18>
      margin=<ltrb 0 0 0 0>
      outlineSettings {
       bUseBrushTransparency=false
       roundingType=FixedRadius
       width=1
       color {colorUseRule=UseColor_ColorTable, specifiedColor=<rgba 0 0 0 0>}
       cornerRadii=<3 3 3 3>
      }
      tintColor {colorUseRule=UseColor_Specified, specifiedColor=<rgba 0.0048 0.0048 0.0048 1>}
      uVRegion {bIsValid=false, max=<0 0>, min=<0 0>}
     }
     backgroundImage {
      bIsDynamicallyLoaded=false
      drawAs=RoundedBox
      imageType=NoImage
      mirroring=NoMirror
      resourceName=None
      resourceObject=null
      tiling=NoTile
      imageSize=<18 18>
      margin=<ltrb 0 0 0 0>
      outlineSettings {
       bUseBrushTransparency=false
       roundingType=FixedRadius
       width=1
       color {colorUseRule=UseColor_ColorTable, specifiedColor=<rgba 0 0 0 0>}
       cornerRadii=<3 3 3 3>
      }
      tintColor {colorUseRule=UseColor_Specified, specifiedColor=<rgba 0.0048 0.0048 0.0048 1>}
      uVRegion {bIsValid=false, max=<0 0>, min=<0 0>}
     }
     backgroundPressedImage {
      bIsDynamicallyLoaded=false
      drawAs=RoundedBox
      imageType=NoImage
      mirroring=NoMirror
      resourceName=None
      resourceObject=null
      tiling=NoTile
      imageSize=<18 18>
      margin=<ltrb 0 0 0 0>
      outlineSettings {
       bUseBrushTransparency=false
       roundingType=FixedRadius
       width=1
       color {colorUseRule=UseColor_ColorTable, specifiedColor=<rgba 0 0 0 0>}
       cornerRadii=<3 3 3 3>
      }
      tintColor {colorUseRule=UseColor_Specified, specifiedColor=<rgba 0.0048 0.0048 0.0048 1>}
      uVRegion {bIsValid=false, max=<0 0>, min=<0 0>}
     }
     borderBackgroundColor {colorUseRule=UseColor_Specified, specifiedColor=<rgba 1 1 1 1>}
     checkedForeground {colorUseRule=UseColor_Specified, specifiedColor=<rgba 1 1 1 1>}
     checkedHoveredForeground {colorUseRule=UseColor_Specified, specifiedColor=<rgba 1 1 1 1>}
     checkedHoveredImage {
      bIsDynamicallyLoaded=false
      drawAs=Image
      imageType=Vector
      mirroring=NoMirror
      resourceName=../../../Engine/Content/Slate/Starship/CoreWidgets/CheckBox/check.svg
      resourceObject=null
      tiling=NoTile
      imageSize=<16 16>
      margin=<ltrb 0 0 0 0>
      outlineSettings {
       bUseBrushTransparency=false
       roundingType=HalfHeightRadius
       width=0
       color {colorUseRule=UseColor_Specified, specifiedColor=<rgba 0 0 0 0>}
       cornerRadii=<0 0 0 0>
      }
      tintColor {colorUseRule=UseColor_Specified, specifiedColor=<rgba 0.0044 0.2384 1 1>}
      uVRegion {bIsValid=false, max=<0 0>, min=<0 0>}
     }
     checkedImage {
      bIsDynamicallyLoaded=false
      drawAs=Image
      imageType=Vector
      mirroring=NoMirror
      resourceName=../../../Engine/Content/Slate/Starship/CoreWidgets/CheckBox/check.svg
      resourceObject=null
      tiling=NoTile
      imageSize=<16 16>
      margin=<ltrb 0 0 0 0>
      outlineSettings {
       bUseBrushTransparency=false
       roundingType=HalfHeightRadius
       width=0
       color {colorUseRule=UseColor_Specified, specifiedColor=<rgba 0 0 0 0>}
       cornerRadii=<0 0 0 0>
      }
      tintColor {colorUseRule=UseColor_Specified, specifiedColor=<rgba 0 0.162 0.7454 1>}
      uVRegion {bIsValid=false, max=<0 0>, min=<0 0>}
     }
     checkedPressedForeground {colorUseRule=UseColor_Specified, specifiedColor=<rgba 1 1 1 1>}
     checkedPressedImage {
      bIsDynamicallyLoaded=false
      drawAs=Image
      imageType=Vector
      mirroring=NoMirror
      resourceName=../../../Engine/Content/Slate/Starship/CoreWidgets/CheckBox/check.svg
      resourceObject=null
      tiling=NoTile
      imageSize=<16 16>
      margin=<ltrb 0 0 0 0>
      outlineSettings {
       bUseBrushTransparency=false
       roundingType=HalfHeightRadius
       width=0
       color {colorUseRule=UseColor_Specified, specifiedColor=<rgba 0 0 0 0>}
       cornerRadii=<0 0 0 0>
      }
      tintColor {colorUseRule=UseColor_Specified, specifiedColor=<rgba 0 0.0802 0.3515 1>}
      uVRegion {bIsValid=false, max=<0 0>, min=<0 0>}
     }
     checkedSlateSound {resourceObject=null}
     foregroundColor {colorUseRule=UseColor_Specified, specifiedColor=<rgba 1 1 1 1>}
     hoveredForeground {colorUseRule=UseColor_Specified, specifiedColor=<rgba 1 1 1 1>}
     hoveredSlateSound {resourceObject=null}
     padding=<ltrb 0 0 0 0>
     pressedForeground {colorUseRule=UseColor_Specified, specifiedColor=<rgba 1 1 1 1>}
     uncheckedHoveredImage {
      bIsDynamicallyLoaded=false
      drawAs=NoDrawType
      imageType=NoImage
      mirroring=NoMirror
      resourceName=None
      resourceObject=null
      tiling=NoTile
      imageSize=<0 0>
      margin=<ltrb 0 0 0 0>
      outlineSettings {
       bUseBrushTransparency=false
       roundingType=HalfHeightRadius
       width=0
       color {colorUseRule=UseColor_Specified, specifiedColor=<rgba 0 0 0 0>}
       cornerRadii=<0 0 0 0>
      }
      tintColor {colorUseRule=UseColor_Specified, specifiedColor=<rgba 1 1 1 1>}
      uVRegion {bIsValid=false, max=<0 0>, min=<0 0>}
     }
     uncheckedImage {
      bIsDynamicallyLoaded=false
      drawAs=NoDrawType
      imageType=NoImage
      mirroring=NoMirror
      resourceName=None
      resourceObject=null
      tiling=NoTile
      imageSize=<0 0>
      margin=<ltrb 0 0 0 0>
      outlineSettings {
       bUseBrushTransparency=false
       roundingType=HalfHeightRadius
       width=0
       color {colorUseRule=UseColor_Specified, specifiedColor=<rgba 0 0 0 0>}
       cornerRadii=<0 0 0 0>
      }
      tintColor {colorUseRule=UseColor_Specified, specifiedColor=<rgba 1 1 1 1>}
      uVRegion {bIsValid=false, max=<0 0>, min=<0 0>}
     }
     uncheckedPressedImage {
      bIsDynamicallyLoaded=false
      drawAs=NoDrawType
      imageType=NoImage
      mirroring=NoMirror
      resourceName=None
      resourceObject=null
      tiling=NoTile
      imageSize=<0 0>
      margin=<ltrb 0 0 0 0>
      outlineSettings {
       bUseBrushTransparency=false
       roundingType=HalfHeightRadius
       width=0
       color {colorUseRule=UseColor_Specified, specifiedColor=<rgba 0 0 0 0>}
       cornerRadii=<0 0 0 0>
      }
      tintColor {colorUseRule=UseColor_Specified, specifiedColor=<rgba 1 1 1 1>}
      uVRegion {bIsValid=false, max=<0 0>, min=<0 0>}
     }
     uncheckedSlateSound {resourceObject=null}
     undeterminedForeground {colorUseRule=UseColor_Specified, specifiedColor=<rgba 1 1 1 1>}
     undeterminedHoveredImage {
      bIsDynamicallyLoaded=false
      drawAs=Image
      imageType=Vector
      mirroring=NoMirror
      resourceName=../../../Engine/Content/Slate/Starship/CoreWidgets/CheckBox/indeterminate.svg
      resourceObject=null
      tiling=NoTile
      imageSize=<16 16>
      margin=<ltrb 0 0 0 0>
      outlineSettings {
       bUseBrushTransparency=false
       roundingType=HalfHeightRadius
       width=0
       color {colorUseRule=UseColor_Specified, specifiedColor=<rgba 0 0 0 0>}
       cornerRadii=<0 0 0 0>
      }
      tintColor {colorUseRule=UseColor_Specified, specifiedColor=<rgba 0.0044 0.2384 1 1>}
      uVRegion {bIsValid=false, max=<0 0>, min=<0 0>}
     }
     undeterminedImage {
      bIsDynamicallyLoaded=false
      drawAs=Image
      imageType=Vector
      mirroring=NoMirror
      resourceName=../../../Engine/Content/Slate/Starship/CoreWidgets/CheckBox/indeterminate.svg
      resourceObject=null
      tiling=NoTile
      imageSize=<16 16>
      margin=<ltrb 0 0 0 0>
      outlineSettings {
       bUseBrushTransparency=false
       roundingType=HalfHeightRadius
       width=0
       color {colorUseRule=UseColor_Specified, specifiedColor=<rgba 0 0 0 0>}
       cornerRadii=<0 0 0 0>
      }
      tintColor {colorUseRule=UseColor_Specified, specifiedColor=<rgba 0 0.162 0.7454 1>}
      uVRegion {bIsValid=false, max=<0 0>, min=<0 0>}
     }
     undeterminedPressedImage {
      bIsDynamicallyLoaded=false
      drawAs=Image
      imageType=Vector
      mirroring=NoMirror
      resourceName=../../../Engine/Content/Slate/Starship/CoreWidgets/CheckBox/indeterminate.svg
      resourceObject=null
      tiling=NoTile
      imageSize=<16 16>
      margin=<ltrb 0 0 0 0>
      outlineSettings {
       bUseBrushTransparency=false
       roundingType=HalfHeightRadius
       width=0
       color {colorUseRule=UseColor_Specified, specifiedColor=<rgba 0 0 0 0>}
       cornerRadii=<0 0 0 0>
      }
      tintColor {colorUseRule=UseColor_Specified, specifiedColor=<rgba 0 0.0802 0.3515 1>}
      uVRegion {bIsValid=false, max=<0 0>, min=<0 0>}
     }
    }
   }
   slot {class=HorizontalBoxSlot}
  }
  HorizontalBox_4 {widgetClassPath=&HorizontalBox, parent=HorizontalBox_10, widget=HorizontalBox_4, slot {class=HorizontalBoxSlot}}
  TextBlock_3 {
   widgetClassPath=&TextBlock
   parent=HorizontalBox_4
   widget=TextBlock_3
   properties {
    text="AnimSequences  will be duplicated in"
    colorAndOpacity {colorUseRule=UseColor_Specified, specifiedColor=<rgba 0.0802 0.1981 0.0648 1>}
    font {
     bForceMonospaced=false
     bMaterialIsStencil=false
     fontMaterial=null
     fontObject=&Roboto
     letterSpacing=0
     monospacedWidth=1
     size=9
     skewAmount=0
     typefaceFontName=Bold
     outlineSettings {
      bApplyOutlineToDropShadows=false
      bMiteredCorners=false
      bSeparateFillAlpha=false
      outlineMaterial=null
      outlineSize=0
      outlineColor=<rgba 0 0 0 1>
     }
    }
   }
   slot {class=HorizontalBoxSlot, properties {verticalAlignment=VAlign_Center}}
  }
  TargetAnimSequencesText {
   widgetClassPath=&TextBlock
   bIsVariable=true
   parent=HorizontalBox_4
   widget=TargetAnimSequencesText
   properties {font {
     bForceMonospaced=false
     bMaterialIsStencil=false
     fontMaterial=null
     fontObject=&Roboto
     letterSpacing=0
     monospacedWidth=1
     size=9
     skewAmount=0
     typefaceFontName=Italic
     outlineSettings {
      bApplyOutlineToDropShadows=false
      bMiteredCorners=false
      bSeparateFillAlpha=false
      outlineMaterial=null
      outlineSize=0
      outlineColor=<rgba 0 0 0 1>
     }
    }}
   slot {class=HorizontalBoxSlot, properties {padding=<ltrb 5 0 0 0>}}
  }
  TargetASSubfolderText {
   widgetClassPath=&TextBlock
   bIsVariable=true
   parent=HorizontalBox_4
   widget=TargetASSubfolderText
   properties {
    text=/Subfolder
    colorAndOpacity {colorUseRule=UseColor_Specified, specifiedColor=<rgba 0.4678 0.2747 0.0908 1>}
    font {
     bForceMonospaced=false
     bMaterialIsStencil=false
     fontMaterial=null
     fontObject=&Roboto
     letterSpacing=0
     monospacedWidth=1
     size=9
     skewAmount=0
     typefaceFontName=Bold Italic
     outlineSettings {
      bApplyOutlineToDropShadows=false
      bMiteredCorners=false
      bSeparateFillAlpha=false
      outlineMaterial=null
      outlineSize=0
      outlineColor=<rgba 0 0 0 1>
     }
    }
   }
   slot {class=HorizontalBoxSlot, properties {verticalAlignment=VAlign_Center}}
  }
  HorizontalBox_16 {
   widgetClassPath=&HorizontalBox
   parent=TargetPathVerticalBox
   widget=HorizontalBox_16
   properties {bOverride_Cursor=true}
   slot {class=VerticalBoxSlot, properties {padding=<ltrb 0 0 0 2>}}
  }
  HorizontalBox_7 {
   widgetClassPath=&HorizontalBox
   parent=HorizontalBox_16
   widget=HorizontalBox_7
   slot {class=HorizontalBoxSlot, properties {padding=<ltrb 0 0 10 0>}}
  }
  TextBlock_16 {
   widgetClassPath=&TextBlock
   parent=HorizontalBox_7
   widget=TextBlock_16
   properties {
    text=Subfolder?
    colorAndOpacity {colorUseRule=UseColor_Specified, specifiedColor=<rgba 0.4678 0.2747 0.0908 1>}
    font {
     bForceMonospaced=false
     bMaterialIsStencil=false
     fontMaterial=null
     fontObject=&Roboto
     letterSpacing=0
     monospacedWidth=1
     size=9
     skewAmount=0
     typefaceFontName=Bold
     outlineSettings {
      bApplyOutlineToDropShadows=false
      bMiteredCorners=false
      bSeparateFillAlpha=false
      outlineMaterial=null
      outlineSize=0
      outlineColor=<rgba 0 0 0 1>
     }
    }
   }
   slot {class=HorizontalBoxSlot, properties {verticalAlignment=VAlign_Center, padding=<ltrb 0 0 5 0>}}
  }
  SubfolderStructureAMCheckbox {
   widgetClassPath=&EditorUtilityCheckBox
   bIsVariable=true
   parent=HorizontalBox_7
   widget=SubfolderStructureAMCheckbox
   properties {
    horizontalAlignment=HAlign_Center
    widgetStyle {
     checkBoxType=CheckBox
     backgroundHoveredImage {
      bIsDynamicallyLoaded=false
      drawAs=RoundedBox
      imageType=NoImage
      mirroring=NoMirror
      resourceName=None
      resourceObject=null
      tiling=NoTile
      imageSize=<18 18>
      margin=<ltrb 0 0 0 0>
      outlineSettings {
       bUseBrushTransparency=false
       roundingType=FixedRadius
       width=1
       color {colorUseRule=UseColor_ColorTable, specifiedColor=<rgba 0 0 0 0>}
       cornerRadii=<3 3 3 3>
      }
      tintColor {colorUseRule=UseColor_Specified, specifiedColor=<rgba 0.0048 0.0048 0.0048 1>}
      uVRegion {bIsValid=false, max=<0 0>, min=<0 0>}
     }
     backgroundImage {
      bIsDynamicallyLoaded=false
      drawAs=RoundedBox
      imageType=NoImage
      mirroring=NoMirror
      resourceName=None
      resourceObject=null
      tiling=NoTile
      imageSize=<18 18>
      margin=<ltrb 0 0 0 0>
      outlineSettings {
       bUseBrushTransparency=false
       roundingType=FixedRadius
       width=1
       color {colorUseRule=UseColor_ColorTable, specifiedColor=<rgba 0 0 0 0>}
       cornerRadii=<3 3 3 3>
      }
      tintColor {colorUseRule=UseColor_Specified, specifiedColor=<rgba 0.0048 0.0048 0.0048 1>}
      uVRegion {bIsValid=false, max=<0 0>, min=<0 0>}
     }
     backgroundPressedImage {
      bIsDynamicallyLoaded=false
      drawAs=RoundedBox
      imageType=NoImage
      mirroring=NoMirror
      resourceName=None
      resourceObject=null
      tiling=NoTile
      imageSize=<18 18>
      margin=<ltrb 0 0 0 0>
      outlineSettings {
       bUseBrushTransparency=false
       roundingType=FixedRadius
       width=1
       color {colorUseRule=UseColor_ColorTable, specifiedColor=<rgba 0 0 0 0>}
       cornerRadii=<3 3 3 3>
      }
      tintColor {colorUseRule=UseColor_Specified, specifiedColor=<rgba 0.0048 0.0048 0.0048 1>}
      uVRegion {bIsValid=false, max=<0 0>, min=<0 0>}
     }
     borderBackgroundColor {colorUseRule=UseColor_Specified, specifiedColor=<rgba 1 1 1 1>}
     checkedForeground {colorUseRule=UseColor_Specified, specifiedColor=<rgba 1 1 1 1>}
     checkedHoveredForeground {colorUseRule=UseColor_Specified, specifiedColor=<rgba 1 1 1 1>}
     checkedHoveredImage {
      bIsDynamicallyLoaded=false
      drawAs=Image
      imageType=Vector
      mirroring=NoMirror
      resourceName=../../../Engine/Content/Slate/Starship/CoreWidgets/CheckBox/check.svg
      resourceObject=null
      tiling=NoTile
      imageSize=<16 16>
      margin=<ltrb 0 0 0 0>
      outlineSettings {
       bUseBrushTransparency=false
       roundingType=HalfHeightRadius
       width=0
       color {colorUseRule=UseColor_Specified, specifiedColor=<rgba 0 0 0 0>}
       cornerRadii=<0 0 0 0>
      }
      tintColor {colorUseRule=UseColor_Specified, specifiedColor=<rgba 0.0044 0.2384 1 1>}
      uVRegion {bIsValid=false, max=<0 0>, min=<0 0>}
     }
     checkedImage {
      bIsDynamicallyLoaded=false
      drawAs=Image
      imageType=Vector
      mirroring=NoMirror
      resourceName=../../../Engine/Content/Slate/Starship/CoreWidgets/CheckBox/check.svg
      resourceObject=null
      tiling=NoTile
      imageSize=<16 16>
      margin=<ltrb 0 0 0 0>
      outlineSettings {
       bUseBrushTransparency=false
       roundingType=HalfHeightRadius
       width=0
       color {colorUseRule=UseColor_Specified, specifiedColor=<rgba 0 0 0 0>}
       cornerRadii=<0 0 0 0>
      }
      tintColor {colorUseRule=UseColor_Specified, specifiedColor=<rgba 0 0.162 0.7454 1>}
      uVRegion {bIsValid=false, max=<0 0>, min=<0 0>}
     }
     checkedPressedForeground {colorUseRule=UseColor_Specified, specifiedColor=<rgba 1 1 1 1>}
     checkedPressedImage {
      bIsDynamicallyLoaded=false
      drawAs=Image
      imageType=Vector
      mirroring=NoMirror
      resourceName=../../../Engine/Content/Slate/Starship/CoreWidgets/CheckBox/check.svg
      resourceObject=null
      tiling=NoTile
      imageSize=<16 16>
      margin=<ltrb 0 0 0 0>
      outlineSettings {
       bUseBrushTransparency=false
       roundingType=HalfHeightRadius
       width=0
       color {colorUseRule=UseColor_Specified, specifiedColor=<rgba 0 0 0 0>}
       cornerRadii=<0 0 0 0>
      }
      tintColor {colorUseRule=UseColor_Specified, specifiedColor=<rgba 0 0.0802 0.3515 1>}
      uVRegion {bIsValid=false, max=<0 0>, min=<0 0>}
     }
     checkedSlateSound {resourceObject=null}
     foregroundColor {colorUseRule=UseColor_Specified, specifiedColor=<rgba 1 1 1 1>}
     hoveredForeground {colorUseRule=UseColor_Specified, specifiedColor=<rgba 1 1 1 1>}
     hoveredSlateSound {resourceObject=null}
     padding=<ltrb 0 0 0 0>
     pressedForeground {colorUseRule=UseColor_Specified, specifiedColor=<rgba 1 1 1 1>}
     uncheckedHoveredImage {
      bIsDynamicallyLoaded=false
      drawAs=NoDrawType
      imageType=NoImage
      mirroring=NoMirror
      resourceName=None
      resourceObject=null
      tiling=NoTile
      imageSize=<0 0>
      margin=<ltrb 0 0 0 0>
      outlineSettings {
       bUseBrushTransparency=false
       roundingType=HalfHeightRadius
       width=0
       color {colorUseRule=UseColor_Specified, specifiedColor=<rgba 0 0 0 0>}
       cornerRadii=<0 0 0 0>
      }
      tintColor {colorUseRule=UseColor_Specified, specifiedColor=<rgba 1 1 1 1>}
      uVRegion {bIsValid=false, max=<0 0>, min=<0 0>}
     }
     uncheckedImage {
      bIsDynamicallyLoaded=false
      drawAs=NoDrawType
      imageType=NoImage
      mirroring=NoMirror
      resourceName=None
      resourceObject=null
      tiling=NoTile
      imageSize=<0 0>
      margin=<ltrb 0 0 0 0>
      outlineSettings {
       bUseBrushTransparency=false
       roundingType=HalfHeightRadius
       width=0
       color {colorUseRule=UseColor_Specified, specifiedColor=<rgba 0 0 0 0>}
       cornerRadii=<0 0 0 0>
      }
      tintColor {colorUseRule=UseColor_Specified, specifiedColor=<rgba 1 1 1 1>}
      uVRegion {bIsValid=false, max=<0 0>, min=<0 0>}
     }
     uncheckedPressedImage {
      bIsDynamicallyLoaded=false
      drawAs=NoDrawType
      imageType=NoImage
      mirroring=NoMirror
      resourceName=None
      resourceObject=null
      tiling=NoTile
      imageSize=<0 0>
      margin=<ltrb 0 0 0 0>
      outlineSettings {
       bUseBrushTransparency=false
       roundingType=HalfHeightRadius
       width=0
       color {colorUseRule=UseColor_Specified, specifiedColor=<rgba 0 0 0 0>}
       cornerRadii=<0 0 0 0>
      }
      tintColor {colorUseRule=UseColor_Specified, specifiedColor=<rgba 1 1 1 1>}
      uVRegion {bIsValid=false, max=<0 0>, min=<0 0>}
     }
     uncheckedSlateSound {resourceObject=null}
     undeterminedForeground {colorUseRule=UseColor_Specified, specifiedColor=<rgba 1 1 1 1>}
     undeterminedHoveredImage {
      bIsDynamicallyLoaded=false
      drawAs=Image
      imageType=Vector
      mirroring=NoMirror
      resourceName=../../../Engine/Content/Slate/Starship/CoreWidgets/CheckBox/indeterminate.svg
      resourceObject=null
      tiling=NoTile
      imageSize=<16 16>
      margin=<ltrb 0 0 0 0>
      outlineSettings {
       bUseBrushTransparency=false
       roundingType=HalfHeightRadius
       width=0
       color {colorUseRule=UseColor_Specified, specifiedColor=<rgba 0 0 0 0>}
       cornerRadii=<0 0 0 0>
      }
      tintColor {colorUseRule=UseColor_Specified, specifiedColor=<rgba 0.0044 0.2384 1 1>}
      uVRegion {bIsValid=false, max=<0 0>, min=<0 0>}
     }
     undeterminedImage {
      bIsDynamicallyLoaded=false
      drawAs=Image
      imageType=Vector
      mirroring=NoMirror
      resourceName=../../../Engine/Content/Slate/Starship/CoreWidgets/CheckBox/indeterminate.svg
      resourceObject=null
      tiling=NoTile
      imageSize=<16 16>
      margin=<ltrb 0 0 0 0>
      outlineSettings {
       bUseBrushTransparency=false
       roundingType=HalfHeightRadius
       width=0
       color {colorUseRule=UseColor_Specified, specifiedColor=<rgba 0 0 0 0>}
       cornerRadii=<0 0 0 0>
      }
      tintColor {colorUseRule=UseColor_Specified, specifiedColor=<rgba 0 0.162 0.7454 1>}
      uVRegion {bIsValid=false, max=<0 0>, min=<0 0>}
     }
     undeterminedPressedImage {
      bIsDynamicallyLoaded=false
      drawAs=Image
      imageType=Vector
      mirroring=NoMirror
      resourceName=../../../Engine/Content/Slate/Starship/CoreWidgets/CheckBox/indeterminate.svg
      resourceObject=null
      tiling=NoTile
      imageSize=<16 16>
      margin=<ltrb 0 0 0 0>
      outlineSettings {
       bUseBrushTransparency=false
       roundingType=HalfHeightRadius
       width=0
       color {colorUseRule=UseColor_Specified, specifiedColor=<rgba 0 0 0 0>}
       cornerRadii=<0 0 0 0>
      }
      tintColor {colorUseRule=UseColor_Specified, specifiedColor=<rgba 0 0.0802 0.3515 1>}
      uVRegion {bIsValid=false, max=<0 0>, min=<0 0>}
     }
    }
   }
   slot {class=HorizontalBoxSlot}
  }
  HorizontalBox_5 {
   widgetClassPath=&HorizontalBox
   parent=HorizontalBox_16
   widget=HorizontalBox_5
   properties {visibility=Visible}
   slot {class=HorizontalBoxSlot}
  }
  TextBlock_4 {
   widgetClassPath=&TextBlock
   parent=HorizontalBox_5
   widget=TextBlock_4
   properties {
    text=AnimMontages will be duplicated in
    colorAndOpacity {colorUseRule=UseColor_Specified, specifiedColor=<rgba 0.1274 0.1274 1 1>}
    font {
     bForceMonospaced=false
     bMaterialIsStencil=false
     fontMaterial=null
     fontObject=&Roboto
     letterSpacing=0
     monospacedWidth=1
     size=9
     skewAmount=0
     typefaceFontName=Bold
     outlineSettings {
      bApplyOutlineToDropShadows=false
      bMiteredCorners=false
      bSeparateFillAlpha=false
      outlineMaterial=null
      outlineSize=0
      outlineColor=<rgba 0 0 0 1>
     }
    }
   }
   slot {class=HorizontalBoxSlot, properties {verticalAlignment=VAlign_Center}}
  }
  TargetAnimMontagesText {
   widgetClassPath=&TextBlock
   bIsVariable=true
   parent=HorizontalBox_5
   widget=TargetAnimMontagesText
   properties {font {
     bForceMonospaced=false
     bMaterialIsStencil=false
     fontMaterial=null
     fontObject=&Roboto
     letterSpacing=0
     monospacedWidth=1
     size=9
     skewAmount=0
     typefaceFontName=Italic
     outlineSettings {
      bApplyOutlineToDropShadows=false
      bMiteredCorners=false
      bSeparateFillAlpha=false
      outlineMaterial=null
      outlineSize=0
      outlineColor=<rgba 0 0 0 1>
     }
    }}
   slot {class=HorizontalBoxSlot, properties {padding=<ltrb 5 0 0 0>}}
  }
  TargetAMSubfolderText {
   widgetClassPath=&TextBlock
   bIsVariable=true
   parent=HorizontalBox_5
   widget=TargetAMSubfolderText
   properties {
    text=/Subfolder
    colorAndOpacity {colorUseRule=UseColor_Specified, specifiedColor=<rgba 0.4678 0.2747 0.0908 1>}
    font {
     bForceMonospaced=false
     bMaterialIsStencil=false
     fontMaterial=null
     fontObject=&Roboto
     letterSpacing=0
     monospacedWidth=1
     size=9
     skewAmount=0
     typefaceFontName=Bold Italic
     outlineSettings {
      bApplyOutlineToDropShadows=false
      bMiteredCorners=false
      bSeparateFillAlpha=false
      outlineMaterial=null
      outlineSize=0
      outlineColor=<rgba 0 0 0 1>
     }
    }
   }
   slot {class=HorizontalBoxSlot, properties {verticalAlignment=VAlign_Center}}
  }
  VerticalBox_7 {
   widgetClassPath=&VerticalBox
   parent=TargetPathVerticalBox
   widget=VerticalBox_7
   slot {class=VerticalBoxSlot, properties {verticalAlignment=VAlign_Bottom, padding=<ltrb 0 20 0 0>, size {sizeRule=Fill, value=1}}}
  }
  TextBlock_11 {
   widgetClassPath=&TextBlock
   parent=VerticalBox_7
   widget=TextBlock_11
   properties {
    text=Target Strings to Replace
    font {
     bForceMonospaced=false
     bMaterialIsStencil=false
     fontMaterial=null
     fontObject=&Roboto
     letterSpacing=0
     monospacedWidth=1
     size=9
     skewAmount=0
     typefaceFontName=Bold
     outlineSettings {
      bApplyOutlineToDropShadows=false
      bMiteredCorners=false
      bSeparateFillAlpha=false
      outlineMaterial=null
      outlineSize=0
      outlineColor=<rgba 0 0 0 1>
     }
    }
   }
   slot {class=VerticalBoxSlot, properties {padding=<ltrb 0 0 0 5>}}
  }
  HorizontalBox_13 {widgetClassPath=&HorizontalBox, parent=VerticalBox_7, widget=HorizontalBox_13, slot {class=VerticalBoxSlot}}
  TextBlock_12 {
   widgetClassPath=&TextBlock
   parent=HorizontalBox_13
   widget=TextBlock_12
   properties {
    text=LevelSequences
    colorAndOpacity {colorUseRule=UseColor_Specified, specifiedColor=<rgba 0.5776 0.0802 0.0802 1>}
    font {
     bForceMonospaced=false
     bMaterialIsStencil=false
     fontMaterial=null
     fontObject=&Roboto
     letterSpacing=0
     monospacedWidth=1
     size=9
     skewAmount=0
     typefaceFontName=Bold
     outlineSettings {
      bApplyOutlineToDropShadows=false
      bMiteredCorners=false
      bSeparateFillAlpha=false
      outlineMaterial=null
      outlineSize=0
      outlineColor=<rgba 0 0 0 1>
     }
    }
   }
   slot {class=HorizontalBoxSlot, properties {verticalAlignment=VAlign_Center}}
  }
  TargetLSTextBox {
   widgetClassPath=&EditorUtilityEditableTextBox
   bIsVariable=true
   parent=HorizontalBox_13
   widget=TargetLSTextBox
   slot {class=HorizontalBoxSlot, properties {padding=<ltrb 10 0 10 0>, size {sizeRule=Fill, value=1}}}
  }
  HorizontalBox_14 {widgetClassPath=&HorizontalBox, parent=VerticalBox_7, widget=HorizontalBox_14, slot {class=VerticalBoxSlot}}
  TextBlock_13 {
   widgetClassPath=&TextBlock
   parent=HorizontalBox_14
   widget=TextBlock_13
   properties {
    text=AnimSequences
    colorAndOpacity {colorUseRule=UseColor_Specified, specifiedColor=<rgba 0.0802 0.1981 0.0648 1>}
    font {
     bForceMonospaced=false
     bMaterialIsStencil=false
     fontMaterial=null
     fontObject=&Roboto
     letterSpacing=0
     monospacedWidth=1
     size=9
     skewAmount=0
     typefaceFontName=Bold
     outlineSettings {
      bApplyOutlineToDropShadows=false
      bMiteredCorners=false
      bSeparateFillAlpha=false
      outlineMaterial=null
      outlineSize=0
      outlineColor=<rgba 0 0 0 1>
     }
    }
   }
   slot {class=HorizontalBoxSlot, properties {verticalAlignment=VAlign_Center}}
  }
  TargetASTextBox {
   widgetClassPath=&EditorUtilityEditableTextBox
   bIsVariable=true
   parent=HorizontalBox_14
   widget=TargetASTextBox
   slot {class=HorizontalBoxSlot, properties {padding=<ltrb 10 0 10 0>, size {sizeRule=Fill, value=1}}}
  }
  HorizontalBox_15 {
   widgetClassPath=&HorizontalBox
   parent=VerticalBox_7
   widget=HorizontalBox_15
   properties {visibility=Visible}
   slot {class=VerticalBoxSlot}
  }
  TextBlock_14 {
   widgetClassPath=&TextBlock
   parent=HorizontalBox_15
   widget=TextBlock_14
   properties {
    text=AnimMontages
    colorAndOpacity {colorUseRule=UseColor_Specified, specifiedColor=<rgba 0.1274 0.1274 1 1>}
    font {
     bForceMonospaced=false
     bMaterialIsStencil=false
     fontMaterial=null
     fontObject=&Roboto
     letterSpacing=0
     monospacedWidth=1
     size=9
     skewAmount=0
     typefaceFontName=Bold
     outlineSettings {
      bApplyOutlineToDropShadows=false
      bMiteredCorners=false
      bSeparateFillAlpha=false
      outlineMaterial=null
      outlineSize=0
      outlineColor=<rgba 0 0 0 1>
     }
    }
   }
   slot {class=HorizontalBoxSlot, properties {verticalAlignment=VAlign_Center}}
  }
  TargetAMTextBox {
   widgetClassPath=&EditorUtilityEditableTextBox
   bIsVariable=true
   parent=HorizontalBox_15
   widget=TargetAMTextBox
   slot {class=HorizontalBoxSlot, properties {padding=<ltrb 10 0 10 0>, size {sizeRule=Fill, value=1}}}
  }
  ListBox {
   widgetClassPath=&VerticalBox
   parent=VerticalBox_94
   widget=ListBox
   properties {bOverride_Cursor=true}
   slot {class=VerticalBoxSlot, properties {size {sizeRule=Fill, value=1}}}
  }
  EditorUtilityScrollBox_4 {
   widgetClassPath=&EditorUtilityScrollBox
   parent=ListBox
   widget=EditorUtilityScrollBox_4
   slot {class=VerticalBoxSlot, properties {size {sizeRule=Fill, value=1}}}
  }
  Border_1 {
   widgetClassPath=&Border
   parent=EditorUtilityScrollBox_4
   widget=Border_1
   properties {
    background {
     bIsDynamicallyLoaded=false
     drawAs=Image
     imageType=NoImage
     mirroring=NoMirror
     resourceName=None
     resourceObject=null
     tiling=NoTile
     imageSize=<32 32>
     margin=<ltrb 0 0 0 0>
     outlineSettings {
      bUseBrushTransparency=false
      roundingType=HalfHeightRadius
      width=0
      color {colorUseRule=UseColor_Specified, specifiedColor=<rgba 0 0 0 0>}
      cornerRadii=<0 0 0 0>
     }
     tintColor {colorUseRule=UseColor_Specified, specifiedColor=<rgba 0.0103 0.0103 0.0103 1>}
     uVRegion {bIsValid=false, max=<0 0>, min=<0 0>}
    }
    padding=<ltrb 0 0 0 0>
   }
   slot {class=ScrollBoxSlot, properties {size {sizeRule=Fill, value=1}}}
  }
  VerticalBox_1 {widgetClassPath=&VerticalBox, parent=Border_1, widget=VerticalBox_1, slot {class=BorderSlot, properties {padding=<ltrb 0 0 0 0>}}}
  HorizontalBox_245 {
   widgetClassPath=&HorizontalBox
   parent=VerticalBox_1
   widget=HorizontalBox_245
   slot {class=VerticalBoxSlot, properties {padding=<ltrb 0 10 0 7>}}
  }
  SelectAllCheckbox {
   widgetClassPath=&EditorUtilityCheckBox
   bIsVariable=true
   parent=HorizontalBox_245
   widget=SelectAllCheckbox
   properties {
    horizontalAlignment=HAlign_Center
    widgetStyle {
     checkBoxType=CheckBox
     backgroundHoveredImage {
      bIsDynamicallyLoaded=false
      drawAs=RoundedBox
      imageType=NoImage
      mirroring=NoMirror
      resourceName=None
      resourceObject=null
      tiling=NoTile
      imageSize=<18 18>
      margin=<ltrb 0 0 0 0>
      outlineSettings {
       bUseBrushTransparency=false
       roundingType=FixedRadius
       width=1
       color {colorUseRule=UseColor_ColorTable, specifiedColor=<rgba 0 0 0 0>}
       cornerRadii=<3 3 3 3>
      }
      tintColor {colorUseRule=UseColor_Specified, specifiedColor=<rgba 0.0048 0.0048 0.0048 1>}
      uVRegion {bIsValid=false, max=<0 0>, min=<0 0>}
     }
     backgroundImage {
      bIsDynamicallyLoaded=false
      drawAs=RoundedBox
      imageType=NoImage
      mirroring=NoMirror
      resourceName=None
      resourceObject=null
      tiling=NoTile
      imageSize=<18 18>
      margin=<ltrb 0 0 0 0>
      outlineSettings {
       bUseBrushTransparency=false
       roundingType=FixedRadius
       width=1
       color {colorUseRule=UseColor_ColorTable, specifiedColor=<rgba 0 0 0 0>}
       cornerRadii=<3 3 3 3>
      }
      tintColor {colorUseRule=UseColor_Specified, specifiedColor=<rgba 0.0048 0.0048 0.0048 1>}
      uVRegion {bIsValid=false, max=<0 0>, min=<0 0>}
     }
     backgroundPressedImage {
      bIsDynamicallyLoaded=false
      drawAs=RoundedBox
      imageType=NoImage
      mirroring=NoMirror
      resourceName=None
      resourceObject=null
      tiling=NoTile
      imageSize=<18 18>
      margin=<ltrb 0 0 0 0>
      outlineSettings {
       bUseBrushTransparency=false
       roundingType=FixedRadius
       width=1
       color {colorUseRule=UseColor_ColorTable, specifiedColor=<rgba 0 0 0 0>}
       cornerRadii=<3 3 3 3>
      }
      tintColor {colorUseRule=UseColor_Specified, specifiedColor=<rgba 0.0048 0.0048 0.0048 1>}
      uVRegion {bIsValid=false, max=<0 0>, min=<0 0>}
     }
     borderBackgroundColor {colorUseRule=UseColor_Specified, specifiedColor=<rgba 1 1 1 1>}
     checkedForeground {colorUseRule=UseColor_Specified, specifiedColor=<rgba 1 1 1 1>}
     checkedHoveredForeground {colorUseRule=UseColor_Specified, specifiedColor=<rgba 1 1 1 1>}
     checkedHoveredImage {
      bIsDynamicallyLoaded=false
      drawAs=Image
      imageType=Vector
      mirroring=NoMirror
      resourceName=../../../Engine/Content/Slate/Starship/CoreWidgets/CheckBox/check.svg
      resourceObject=null
      tiling=NoTile
      imageSize=<16 16>
      margin=<ltrb 0 0 0 0>
      outlineSettings {
       bUseBrushTransparency=false
       roundingType=HalfHeightRadius
       width=0
       color {colorUseRule=UseColor_Specified, specifiedColor=<rgba 0 0 0 0>}
       cornerRadii=<0 0 0 0>
      }
      tintColor {colorUseRule=UseColor_Specified, specifiedColor=<rgba 0.0044 0.2384 1 1>}
      uVRegion {bIsValid=false, max=<0 0>, min=<0 0>}
     }
     checkedImage {
      bIsDynamicallyLoaded=false
      drawAs=Image
      imageType=Vector
      mirroring=NoMirror
      resourceName=../../../Engine/Content/Slate/Starship/CoreWidgets/CheckBox/check.svg
      resourceObject=null
      tiling=NoTile
      imageSize=<16 16>
      margin=<ltrb 0 0 0 0>
      outlineSettings {
       bUseBrushTransparency=false
       roundingType=HalfHeightRadius
       width=0
       color {colorUseRule=UseColor_Specified, specifiedColor=<rgba 0 0 0 0>}
       cornerRadii=<0 0 0 0>
      }
      tintColor {colorUseRule=UseColor_Specified, specifiedColor=<rgba 0 0.162 0.7454 1>}
      uVRegion {bIsValid=false, max=<0 0>, min=<0 0>}
     }
     checkedPressedForeground {colorUseRule=UseColor_Specified, specifiedColor=<rgba 1 1 1 1>}
     checkedPressedImage {
      bIsDynamicallyLoaded=false
      drawAs=Image
      imageType=Vector
      mirroring=NoMirror
      resourceName=../../../Engine/Content/Slate/Starship/CoreWidgets/CheckBox/check.svg
      resourceObject=null
      tiling=NoTile
      imageSize=<16 16>
      margin=<ltrb 0 0 0 0>
      outlineSettings {
       bUseBrushTransparency=false
       roundingType=HalfHeightRadius
       width=0
       color {colorUseRule=UseColor_Specified, specifiedColor=<rgba 0 0 0 0>}
       cornerRadii=<0 0 0 0>
      }
      tintColor {colorUseRule=UseColor_Specified, specifiedColor=<rgba 0 0.0802 0.3515 1>}
      uVRegion {bIsValid=false, max=<0 0>, min=<0 0>}
     }
     checkedSlateSound {resourceObject=null}
     foregroundColor {colorUseRule=UseColor_Specified, specifiedColor=<rgba 1 1 1 1>}
     hoveredForeground {colorUseRule=UseColor_Specified, specifiedColor=<rgba 1 1 1 1>}
     hoveredSlateSound {resourceObject=null}
     padding=<ltrb 0 0 0 0>
     pressedForeground {colorUseRule=UseColor_Specified, specifiedColor=<rgba 1 1 1 1>}
     uncheckedHoveredImage {
      bIsDynamicallyLoaded=false
      drawAs=NoDrawType
      imageType=NoImage
      mirroring=NoMirror
      resourceName=None
      resourceObject=null
      tiling=NoTile
      imageSize=<0 0>
      margin=<ltrb 0 0 0 0>
      outlineSettings {
       bUseBrushTransparency=false
       roundingType=HalfHeightRadius
       width=0
       color {colorUseRule=UseColor_Specified, specifiedColor=<rgba 0 0 0 0>}
       cornerRadii=<0 0 0 0>
      }
      tintColor {colorUseRule=UseColor_Specified, specifiedColor=<rgba 1 1 1 1>}
      uVRegion {bIsValid=false, max=<0 0>, min=<0 0>}
     }
     uncheckedImage {
      bIsDynamicallyLoaded=false
      drawAs=NoDrawType
      imageType=NoImage
      mirroring=NoMirror
      resourceName=None
      resourceObject=null
      tiling=NoTile
      imageSize=<0 0>
      margin=<ltrb 0 0 0 0>
      outlineSettings {
       bUseBrushTransparency=false
       roundingType=HalfHeightRadius
       width=0
       color {colorUseRule=UseColor_Specified, specifiedColor=<rgba 0 0 0 0>}
       cornerRadii=<0 0 0 0>
      }
      tintColor {colorUseRule=UseColor_Specified, specifiedColor=<rgba 1 1 1 1>}
      uVRegion {bIsValid=false, max=<0 0>, min=<0 0>}
     }
     uncheckedPressedImage {
      bIsDynamicallyLoaded=false
      drawAs=NoDrawType
      imageType=NoImage
      mirroring=NoMirror
      resourceName=None
      resourceObject=null
      tiling=NoTile
      imageSize=<0 0>
      margin=<ltrb 0 0 0 0>
      outlineSettings {
       bUseBrushTransparency=false
       roundingType=HalfHeightRadius
       width=0
       color {colorUseRule=UseColor_Specified, specifiedColor=<rgba 0 0 0 0>}
       cornerRadii=<0 0 0 0>
      }
      tintColor {colorUseRule=UseColor_Specified, specifiedColor=<rgba 1 1 1 1>}
      uVRegion {bIsValid=false, max=<0 0>, min=<0 0>}
     }
     uncheckedSlateSound {resourceObject=null}
     undeterminedForeground {colorUseRule=UseColor_Specified, specifiedColor=<rgba 1 1 1 1>}
     undeterminedHoveredImage {
      bIsDynamicallyLoaded=false
      drawAs=Image
      imageType=Vector
      mirroring=NoMirror
      resourceName=../../../Engine/Content/Slate/Starship/CoreWidgets/CheckBox/indeterminate.svg
      resourceObject=null
      tiling=NoTile
      imageSize=<16 16>
      margin=<ltrb 0 0 0 0>
      outlineSettings {
       bUseBrushTransparency=false
       roundingType=HalfHeightRadius
       width=0
       color {colorUseRule=UseColor_Specified, specifiedColor=<rgba 0 0 0 0>}
       cornerRadii=<0 0 0 0>
      }
      tintColor {colorUseRule=UseColor_Specified, specifiedColor=<rgba 0.0044 0.2384 1 1>}
      uVRegion {bIsValid=false, max=<0 0>, min=<0 0>}
     }
     undeterminedImage {
      bIsDynamicallyLoaded=false
      drawAs=Image
      imageType=Vector
      mirroring=NoMirror
      resourceName=../../../Engine/Content/Slate/Starship/CoreWidgets/CheckBox/indeterminate.svg
      resourceObject=null
      tiling=NoTile
      imageSize=<16 16>
      margin=<ltrb 0 0 0 0>
      outlineSettings {
       bUseBrushTransparency=false
       roundingType=HalfHeightRadius
       width=0
       color {colorUseRule=UseColor_Specified, specifiedColor=<rgba 0 0 0 0>}
       cornerRadii=<0 0 0 0>
      }
      tintColor {colorUseRule=UseColor_Specified, specifiedColor=<rgba 0 0.162 0.7454 1>}
      uVRegion {bIsValid=false, max=<0 0>, min=<0 0>}
     }
     undeterminedPressedImage {
      bIsDynamicallyLoaded=false
      drawAs=Image
      imageType=Vector
      mirroring=NoMirror
      resourceName=../../../Engine/Content/Slate/Starship/CoreWidgets/CheckBox/indeterminate.svg
      resourceObject=null
      tiling=NoTile
      imageSize=<16 16>
      margin=<ltrb 0 0 0 0>
      outlineSettings {
       bUseBrushTransparency=false
       roundingType=HalfHeightRadius
       width=0
       color {colorUseRule=UseColor_Specified, specifiedColor=<rgba 0 0 0 0>}
       cornerRadii=<0 0 0 0>
      }
      tintColor {colorUseRule=UseColor_Specified, specifiedColor=<rgba 0 0.0802 0.3515 1>}
      uVRegion {bIsValid=false, max=<0 0>, min=<0 0>}
     }
    }
   }
   slot {class=HorizontalBoxSlot, properties {horizontalAlignment=HAlign_Center, padding=<ltrb 10 0 10 0>}}
  }
  TextBlock_137 {
   widgetClassPath=&TextBlock
   parent=HorizontalBox_245
   widget=TextBlock_137
   properties {
    text=Source Animations
    font {
     bForceMonospaced=false
     bMaterialIsStencil=false
     fontMaterial=null
     fontObject=&Roboto
     letterSpacing=0
     monospacedWidth=1
     size=13.5
     skewAmount=0
     typefaceFontName=Bold
     outlineSettings {
      bApplyOutlineToDropShadows=false
      bMiteredCorners=false
      bSeparateFillAlpha=false
      outlineMaterial=null
      outlineSize=0
      outlineColor=<rgba 0 0 0 1>
     }
    }
   }
   slot {class=HorizontalBoxSlot, properties {padding=<ltrb 10 0 0 0>, size {sizeRule=Fill, value=1}}}
  }
  TextBlock_15 {
   widgetClassPath=&TextBlock
   parent=HorizontalBox_245
   widget=TextBlock_15
   properties {
    text=Target Animations
    font {
     bForceMonospaced=false
     bMaterialIsStencil=false
     fontMaterial=null
     fontObject=&Roboto
     letterSpacing=0
     monospacedWidth=1
     size=13.5
     skewAmount=0
     typefaceFontName=Bold
     outlineSettings {
      bApplyOutlineToDropShadows=false
      bMiteredCorners=false
      bSeparateFillAlpha=false
      outlineMaterial=null
      outlineSize=0
      outlineColor=<rgba 0 0 0 1>
     }
    }
   }
   slot {class=HorizontalBoxSlot, properties {padding=<ltrb 38 0 0 0>, size {sizeRule=Fill, value=1}}}
  }
  LevelSequenceListView {
   widgetClassPath=/Script/Blutility.EditorUtilityListView
   bIsVariable=true
   parent=VerticalBox_1
   widget=LevelSequenceListView
   properties {
    allowOverscroll=false
    bClearSelectionOnClick=true
    bEnableRightClickScrolling=false
    entryWidgetClass=/AnimInEdTools/AnimBatchDuplicate/Widgets/EUW_AnimBatchDup_LS_Item.EUW_AnimBatchDup_LS_Item_C
    numDesignerPreviewEntries=1
   }
   slot {class=VerticalBoxSlot}
  }
  DuplicateButtonBox {widgetClassPath=&VerticalBox, bIsVariable=true, parent=VerticalBox_94, widget=DuplicateButtonBox, slot {class=VerticalBoxSlot}}
  DuplicateButton {
   widgetClassPath=&EditorUtilityButton
   parent=DuplicateButtonBox
   widget=DuplicateButton
   properties {toolTipText=Duplicates the selected animations to the target AC paths}
   slot {class=VerticalBoxSlot}
  }
  TextBlock_90 {widgetClassPath=&TextBlock, parent=DuplicateButton, widget=TextBlock_90, properties {text=Duplicate}, slot {class=ButtonSlot}}
 ]
}
