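# Editor utility list row for the Anim Batch Duplicate tool, showing one subfolder item and its nested assets.
# Setting the item fills a sub-list with its Level Sequence, Anim Sequences and Montages, deselecting them via all-false include flags.
# The checkbox sets item selection, row tint, target-folder visibility and displayed sub-entry checkboxes.
# A separate helper, uncalled here, collects existing AnimSequence paths linked to a Level Sequence.
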EditorUtilityWidgetBlueprint /AnimInEdTools/AnimBatchDuplicate/Widgets/EUW_AnimBatchDup_Entry
graphs [
 EventGraph {nodes [
   n1 = "On Entry Initialized (SubEntryListView)"!() -> (OutputDelegate, then, Item:Object Reference, Widget:User Widget Object Reference) {is=title K2Node_ComponentBoundEvent}
   > n2 = Cast To BP_AnimBatchDup_SubItem!(execute, Object=@n1.Item) -> (then, CastFailed, AsBP Anim Batch Dup Sub Item:BP Anim Batch Dup Sub Item Object Reference) {is=title K2Node_DynamicCast}
   > n3 = Cast To EUW_AnimBatchDup_SubEntry!(execute, Object=@n1.Widget) -> (then, CastFailed, AsEUW Anim Batch Dup Sub Entry:EUW Anim Batch Dup Sub Entry Object Reference) {is=title K2Node_DynamicCast}
   n4 = Get Selected!(self:BP Anim Batch Dup Sub Item Object Reference=@n2) -> (Selected:bool)
   > Update Checkbox!(execute, self:EUW Anim Batch Dup Sub Entry Object Reference=@n3, Checked:bool=@n4) -> (then)
   > Show/Hide Checkbox!(execute, self:EUW Anim Batch Dup Sub Entry Object Reference=@n3, Checked:bool=@n4) -> (then)
   > n5 = Branch(Condition=@n4)
   Cancel Edit Text(execute=@n5.else, self:EUW Anim Batch Dup Sub Entry Object Reference=@n3)
   n6 = "On Check State Changed (EntryCheckbox)"!() -> (OutputDelegate, then, bIsChecked:bool) {is=title K2Node_ComponentBoundEvent}
   > Update Entry Checkbox!(execute, self, Checked:bool=@n6.bIsChecked) -> (then)
   n7 = Update Entry Checkbox!() -> (OutputDelegate, then, Checked:bool=~) {is=title K2Node_CustomEvent}
   > n8 = Sequence()
   > Set Selected!(
    execute
    Selected:bool=@n7.Checked
    self:BP Anim Batch Dup Item Object Reference=Get Current Item() -> (Current Item:BP Anim Batch Dup Item Object Reference)
   ) -> (then, Output_Get:bool)
   > SetCheckedState!(
    execute
    self:Check Box Object Reference=Get EntryCheckbox() -> (EntryCheckbox:Editor Utility Check Box Object Reference)
    InCheckedState:ECheckBoxState Enum=Select(Option 0:ECheckBoxState Enum=Unchecked, Option 1:ECheckBoxState Enum=Checked, Index=@n7.Checked) -> (ReturnValue:ECheckBoxState Enum)
   ) -> (then)
   > SetBrushColor(
    self=Get EntryBorder() -> (EntryBorder:Border Object Reference)
    InBrushColor=SelectColor(A=(R=0.025,G=0.025,B=0.025,A=1), B=(R=0.0176,G=0.0176,B=0.0176,A=0), bPickA=@n7.Checked)
   )
   > n9 = For Each Loop(Array:Array of Object References=GetDisplayedEntryWidgets!(self:List View Base Object Reference=Get SubEntryListView() -> (SubEntryListView:Editor Utility List View Object Reference)) -> (ReturnValue:Array of User Widget Object References)) -> (Array Element:Object Reference)
   > n10 = Cast To EUW_AnimBatchDup_SubEntry(Object=@n9)
   > Update Checkbox(self=@n10, Checked=@n7.Checked)
   > Show/Hide Checkbox(self=@n10, Checked=@n7.Checked)
   SetVisibility(
    execute=@n8.then_1
    self=Get TargetSubfolderText() -> (TargetSubfolderText:Text Object Reference)
    InVisibility=Select(Option 0:ESlateVisibility Enum=Collapsed, Option 1:ESlateVisibility Enum=Visible, Index=@n7.Checked) -> (ReturnValue:ESlateVisibility Enum)
   )
   > SetVisibility(
    self=Get TargetFolderIcon() -> (TargetFolderIcon:Image Object Reference)
    InVisibility=Select(Option 0=Collapsed, Option 1=Visible, Index=@n7.Checked)
   )
   n11 = Populate SubItems ListView!() -> (OutputDelegate, then, Include AM:bool=~, Include AS:bool=~, Include LS:bool=~) {is=title K2Node_CustomEvent}
   > n12 = Sequence!(execute) -> (then_0, then_1, then_2:Exec, then_3:Exec)
   Update Entry Checkbox(execute=@n12.then_2, Checked=Get Selected(self:BP Anim Batch Dup Item Object Reference=Get Current Item())) {is=title K2Node_CallFunction}
   n13 = Get Level Sequence Item!(self:BP Anim Batch Dup Item Object Reference=Get Current Item()) -> (Level Sequence Item:BP Anim Batch Dup Sub Item LS Object Reference)
   n14 = Set Text(
    execute=@n12.then_3
    Text="To Text (String)"(InString=Get Subfolder!(self:BP Anim Batch Dup Sub Item Object Reference=@n13) -> (Subfolder:String) {is=title K2Node_CallFunction})
    self=Get SubfolderText() -> (SubfolderText:Text Object Reference)
   )
   > Set Text(Text=@n14, self=Get TargetSubfolderText())
   n15 = Get SubEntryListView()
   ClearListItems(execute=@n12.then_0, self=@n15)
   > RegenerateAllEntries(self:List View Base Object Reference=@n15)
   > Update SubItems Selection!(execute, self, Include AM:bool=@n11.Include AM, Include AS:bool=@n11.Include AS, Include LS:bool=@n11.Include LS) -> (then)
   AddItem!(
    execute=@n12.then_1
    self:List View Object Reference=Get SubEntryListView()
    Item:Object Reference=Get Level Sequence Item(self=Get Current Item())
   ) -> (then)
   > n16 = Sequence!(execute) -> (then_0, then_1)
   n17 = For Each Loop(
    Exec=@n16.then_1
    Array:Array of BP Anim Batch Dup Sub Item AM Object References=Get Anim Montage Items!(self:BP Anim Batch Dup Item Object Reference=Get Current Item()) -> (Anim Montage Items:Array of BP Anim Batch Dup Sub Item AM Object References)
   ) -> (Array Element:BP Anim Batch Dup Sub Item AM Object Reference)
   > AddItem(self=Get SubEntryListView(), Item=@n17)
   n18 = For Each Loop(
    Exec=@n16.then_0
    Array:Array of BP Anim Batch Dup Sub Item AS Object References=Get Anim Sequence Items!(self:BP Anim Batch Dup Item Object Reference=Get Current Item()) -> (Anim Sequence Items:Array of BP Anim Batch Dup Sub Item AS Object References)
   ) -> (Array Element:BP Anim Batch Dup Sub Item AS Object Reference)
   > AddItem(self=Get SubEntryListView(), Item=@n18)
   n19 = Update SubItems Selection!() -> (OutputDelegate, then, Include AM:bool=~, Include AS:bool=~, Include LS:bool=~) {is=title K2Node_CustomEvent}
   > n20 = Sequence!(execute) -> (then_0, then_1, then_2:Exec)
   n21 = For Each Loop(Exec=@n20.then_1, Array=Get Anim Sequence Items(self=Get Current Item()))
   n22 = Get Current Item()
   > Set Selected(Selected=AND Boolean(A=Get Selected(self=@n22), B=@n19.Include AS), self:BP Anim Batch Dup Sub Item Object Reference=@n21)
   n23 = For Each Loop(Exec=@n20.then_2, Array:Array of BP Anim Batch Dup Sub Item AM Object References=Get Anim Montage Items(self=Get Current Item())) -> (Array Element:BP Anim Batch Dup Sub Item AM Object Reference)
   n24 = Get Current Item()
   > Set Selected(Selected=AND Boolean(A=Get Selected(self=@n24), B=@n19.Include AM), self=@n23)
   n25 = Get Current Item()
   Set Selected(
    execute=@n20.then_0
    Selected=AND Boolean(A=Get Selected(self=@n25), B=@n19.Include LS)
    self=Get Level Sequence Item(self=Get Current Item())
   )
   n26 = Event OnListItemObjectSet()
   > n27 = Cast To BP_AnimBatchDup_Item!(execute, Object=@n26.ListItemObject) -> (then, CastFailed, AsBP Anim Batch Dup Item:BP Anim Batch Dup Item Object Reference) {is=title K2Node_DynamicCast}
   > Set Current Item(Current Item:BP Anim Batch Dup Item Object Reference=@n27) -> (Output_Get:BP Anim Batch Dup Item Object Reference)
   > Populate SubItems ListView!(execute, self, Include AM:bool=false, Include AS:bool=false, Include LS:bool=false) -> (then) {is=title K2Node_CallFunction}
  ]}
 Get Linked AnimSeq Names from LevelSeq {nodes [
   n1 = Get Linked AnimSeq Names from LevelSeq!() -> (then, Level Sequence:Level Sequence Object Reference) {is=title K2Node_FunctionEntry}
   > n2 = Sequence!(execute) -> (then_0, then_1)
   > n3 = GetAnimSequenceLinkFromLevelSequence!(execute, InLevelSequence:Level Sequence Object Reference=@n1) -> (then, ReturnValue:Level Sequence Anim Sequence Link Object Reference)
   > n4 = For Each Loop(Array:Array of Level Sequence Anim Sequence Link Item Structures=Get AnimSequenceLinks!(self:Level Sequence Anim Sequence Link Object Reference=@n3) -> (AnimSequenceLinks:Array of Level Sequence Anim Sequence Link Item Structures)) -> (Array Element:Level Sequence Anim Sequence Link Item Structure)
   n5 = Break Level Sequence Anim Sequence Link Item!(LevelSequenceAnimSequenceLinkItem:"Level Sequence Anim Sequence Link Item Structure (by ref)"=@n4) -> (
    SkelTrackGuid:Guid Structure
    PathToAnimSequence_PathString:String
    bAutoBake:bool=~
    bExportTransforms:bool=~
    bExportMorphTargets:bool=~
    bExportAttributeCurves:bool=~
    bExportMaterialCurves:bool=~
    bSkipCurvesWithZeroValue:bool=~
    bSetRetargetSourceAsset:bool=~
    Interpolation:EAnimInterpolationType Enum
    CurveInterpolation:ERichCurveInterpMode Enum
    bRecordInWorldSpace:bool=~
    bEvaluateAllSkeletalMeshComponents:bool=~
    IncludeAnimationNames:Array of Strings
    ExcludeAnimationNames:Array of Strings
    bRemoveExcludedCurves:bool=~
    WarmUpFrames:Frame Number Structure
    DelayBeforeStart:Frame Number Structure
    bUseCustomTimeRange:bool=~
    CustomStartFrame:Frame Number Structure
    CustomEndFrame:Frame Number Structure
    CustomDisplayRate:Frame Rate Structure
    bUseCustomFrameRate:bool=~
    CustomFrameRate:Frame Rate Structure
   ) {is=title K2Node_BreakStruct}
   > n6 = DoesAssetExist(AssetPath=@n5.PathToAnimSequence_PathString)
   > Branch(Condition=@n6)
   > Add(TargetArray=Get Local AnimSequences Names() -> (Local AnimSequences Names:Array of Strings), NewItem=@n5.PathToAnimSequence_PathString) {is=title K2Node_CallArrayFunction}
   Return Node!(execute=@n2.then_1, AnimSequences Names:Array of Strings=Get Local AnimSequences Names())
  ]}
]
variables [
 AC {type=AnimatableCreature_C, default=null}
 Current Item {type=BP_AnimBatchDup_Item_C, default=null}
 Path {type=string, default=""}
]
widgets {
 info {rootWidgetClass=&StackBox, widgetCount=11}
 widgets [
  StackBox_0 {widgetClassPath=&StackBox, widget=StackBox_0}
  EntryBorder {
   widgetClassPath=&Border
   bIsVariable=true
   parent=StackBox_0
   widget=EntryBorder
   properties {brushColor=<rgba 0.0203 0.0203 0.0203 1>, padding=<ltrb 0 5 0 5>}
   slot {class=StackBoxSlot, properties {verticalAlignment=VAlign_Center, padding=<ltrb 0 2 0 2>, size {sizeRule=Fill, value=1}}}
  }
  HorizontalBox_0 {
   widgetClassPath=&HorizontalBox
   parent=EntryBorder
   widget=HorizontalBox_0
   slot {class=BorderSlot, properties {padding=<ltrb 0 5 0 5>}}
  }
  EntryCheckbox {
   widgetClassPath=&EditorUtilityCheckBox
   bIsVariable=true
   parent=HorizontalBox_0
   widget=EntryCheckbox
   properties {
    checkedState=Checked
    widgetStyle {
     checkBoxType=CheckBox
     backgroundHoveredImage {
      bIsDynamicallyLoaded=false
      drawAs=RoundedBox
      imageType=NoImage
      mirroring=NoMirror
      resourceName=None
      resourceObject=null
      tiling=NoTile
      imageSize=<18 18>
      margin=<ltrb 0 0 0 0>
      outlineSettings {
       bUseBrushTransparency=false
       roundingType=FixedRadius
       width=1
       color {colorUseRule=UseColor_ColorTable, specifiedColor=<rgba 0 0 0 0>}
       cornerRadii=<3 3 3 3>
      }
      tintColor {colorUseRule=UseColor_Specified, specifiedColor=<rgba 0.0048 0.0048 0.0048 1>}
      uVRegion {bIsValid=false, max=<0 0>, min=<0 0>}
     }
     backgroundImage {
      bIsDynamicallyLoaded=false
      drawAs=RoundedBox
      imageType=NoImage
      mirroring=NoMirror
      resourceName=None
      resourceObject=null
      tiling=NoTile
      imageSize=<18 18>
      margin=<ltrb 0 0 0 0>
      outlineSettings {
       bUseBrushTransparency=false
       roundingType=FixedRadius
       width=1
       color {colorUseRule=UseColor_ColorTable, specifiedColor=<rgba 0 0 0 0>}
       cornerRadii=<3 3 3 3>
      }
      tintColor {colorUseRule=UseColor_Specified, specifiedColor=<rgba 0.0048 0.0048 0.0048 1>}
      uVRegion {bIsValid=false, max=<0 0>, min=<0 0>}
     }
     backgroundPressedImage {
      bIsDynamicallyLoaded=false
      drawAs=RoundedBox
      imageType=NoImage
      mirroring=NoMirror
      resourceName=None
      resourceObject=null
      tiling=NoTile
      imageSize=<18 18>
      margin=<ltrb 0 0 0 0>
      outlineSettings {
       bUseBrushTransparency=false
       roundingType=FixedRadius
       width=1
       color {colorUseRule=UseColor_ColorTable, specifiedColor=<rgba 0 0 0 0>}
       cornerRadii=<3 3 3 3>
      }
      tintColor {colorUseRule=UseColor_Specified, specifiedColor=<rgba 0.0048 0.0048 0.0048 1>}
      uVRegion {bIsValid=false, max=<0 0>, min=<0 0>}
     }
     borderBackgroundColor {colorUseRule=UseColor_Specified, specifiedColor=<rgba 1 1 1 1>}
     checkedForeground {colorUseRule=UseColor_Specified, specifiedColor=<rgba 1 1 1 1>}
     checkedHoveredForeground {colorUseRule=UseColor_Specified, specifiedColor=<rgba 1 1 1 1>}
     checkedHoveredImage {
      bIsDynamicallyLoaded=false
      drawAs=Image
      imageType=Vector
      mirroring=NoMirror
      resourceName=../../../Engine/Content/Slate/Starship/CoreWidgets/CheckBox/check.svg
      resourceObject=null
      tiling=NoTile
      imageSize=<25 25>
      margin=<ltrb 0 0 0 0>
      outlineSettings {
       bUseBrushTransparency=false
       roundingType=HalfHeightRadius
       width=0
       color {colorUseRule=UseColor_Specified, specifiedColor=<rgba 0 0 0 0>}
       cornerRadii=<0 0 0 0>
      }
      tintColor {colorUseRule=UseColor_Specified, specifiedColor=<rgba 0.0044 0.2384 1 1>}
      uVRegion {bIsValid=false, max=<0 0>, min=<0 0>}
     }
     checkedImage {
      bIsDynamicallyLoaded=false
      drawAs=Image
      imageType=Vector
      mirroring=NoMirror
      resourceName=../../../Engine/Content/Slate/Starship/CoreWidgets/CheckBox/check.svg
      resourceObject=null
      tiling=NoTile
      imageSize=<25 25>
      margin=<ltrb 0 0 0 0>
      outlineSettings {
       bUseBrushTransparency=false
       roundingType=HalfHeightRadius
       width=0
       color {colorUseRule=UseColor_Specified, specifiedColor=<rgba 0 0 0 0>}
       cornerRadii=<0 0 0 0>
      }
      tintColor {colorUseRule=UseColor_Specified, specifiedColor=<rgba 0 0.162 0.7454 1>}
      uVRegion {bIsValid=false, max=<0 0>, min=<0 0>}
     }
     checkedPressedForeground {colorUseRule=UseColor_Specified, specifiedColor=<rgba 1 1 1 1>}
     checkedPressedImage {
      bIsDynamicallyLoaded=false
      drawAs=Image
      imageType=Vector
      mirroring=NoMirror
      resourceName=../../../Engine/Content/Slate/Starship/CoreWidgets/CheckBox/check.svg
      resourceObject=null
      tiling=NoTile
      imageSize=<25 25>
      margin=<ltrb 0 0 0 0>
      outlineSettings {
       bUseBrushTransparency=false
       roundingType=HalfHeightRadius
       width=0
       color {colorUseRule=UseColor_Specified, specifiedColor=<rgba 0 0 0 0>}
       cornerRadii=<0 0 0 0>
      }
      tintColor {colorUseRule=UseColor_Specified, specifiedColor=<rgba 0 0.0802 0.3515 1>}
      uVRegion {bIsValid=false, max=<0 0>, min=<0 0>}
     }
     checkedSlateSound {resourceObject=null}
     foregroundColor {colorUseRule=UseColor_Specified, specifiedColor=<rgba 1 1 1 1>}
     hoveredForeground {colorUseRule=UseColor_Specified, specifiedColor=<rgba 1 1 1 1>}
     hoveredSlateSound {resourceObject=null}
     padding=<ltrb 0 0 0 0>
     pressedForeground {colorUseRule=UseColor_Specified, specifiedColor=<rgba 1 1 1 1>}
     uncheckedHoveredImage {
      bIsDynamicallyLoaded=false
      drawAs=NoDrawType
      imageType=NoImage
      mirroring=NoMirror
      resourceName=None
      resourceObject=null
      tiling=NoTile
      imageSize=<25 25>
      margin=<ltrb 0 0 0 0>
      outlineSettings {
       bUseBrushTransparency=false
       roundingType=HalfHeightRadius
       width=0
       color {colorUseRule=UseColor_Specified, specifiedColor=<rgba 0 0 0 0>}
       cornerRadii=<0 0 0 0>
      }
      tintColor {colorUseRule=UseColor_Specified, specifiedColor=<rgba 1 1 1 1>}
      uVRegion {bIsValid=false, max=<0 0>, min=<0 0>}
     }
     uncheckedImage {
      bIsDynamicallyLoaded=false
      drawAs=NoDrawType
      imageType=NoImage
      mirroring=NoMirror
      resourceName=None
      resourceObject=null
      tiling=NoTile
      imageSize=<25 25>
      margin=<ltrb 0 0 0 0>
      outlineSettings {
       bUseBrushTransparency=false
       roundingType=HalfHeightRadius
       width=0
       color {colorUseRule=UseColor_Specified, specifiedColor=<rgba 0 0 0 0>}
       cornerRadii=<0 0 0 0>
      }
      tintColor {colorUseRule=UseColor_Specified, specifiedColor=<rgba 1 1 1 1>}
      uVRegion {bIsValid=false, max=<0 0>, min=<0 0>}
     }
     uncheckedPressedImage {
      bIsDynamicallyLoaded=false
      drawAs=NoDrawType
      imageType=NoImage
      mirroring=NoMirror
      resourceName=None
      resourceObject=null
      tiling=NoTile
      imageSize=<25 25>
      margin=<ltrb 0 0 0 0>
      outlineSettings {
       bUseBrushTransparency=false
       roundingType=HalfHeightRadius
       width=0
       color {colorUseRule=UseColor_Specified, specifiedColor=<rgba 0 0 0 0>}
       cornerRadii=<0 0 0 0>
      }
      tintColor {colorUseRule=UseColor_Specified, specifiedColor=<rgba 1 1 1 1>}
      uVRegion {bIsValid=false, max=<0 0>, min=<0 0>}
     }
     uncheckedSlateSound {resourceObject=null}
     undeterminedForeground {colorUseRule=UseColor_Specified, specifiedColor=<rgba 1 1 1 1>}
     undeterminedHoveredImage {
      bIsDynamicallyLoaded=false
      drawAs=Image
      imageType=Vector
      mirroring=NoMirror
      resourceName=../../../Engine/Content/Slate/Starship/CoreWidgets/CheckBox/indeterminate.svg
      resourceObject=null
      tiling=NoTile
      imageSize=<25 25>
      margin=<ltrb 0 0 0 0>
      outlineSettings {
       bUseBrushTransparency=false
       roundingType=HalfHeightRadius
       width=0
       color {colorUseRule=UseColor_Specified, specifiedColor=<rgba 0 0 0 0>}
       cornerRadii=<0 0 0 0>
      }
      tintColor {colorUseRule=UseColor_Specified, specifiedColor=<rgba 0.0044 0.2384 1 1>}
      uVRegion {bIsValid=false, max=<0 0>, min=<0 0>}
     }
     undeterminedImage {
      bIsDynamicallyLoaded=false
      drawAs=Image
      imageType=Vector
      mirroring=NoMirror
      resourceName=../../../Engine/Content/Slate/Starship/CoreWidgets/CheckBox/indeterminate.svg
      resourceObject=null
      tiling=NoTile
      imageSize=<25 25>
      margin=<ltrb 0 0 0 0>
      outlineSettings {
       bUseBrushTransparency=false
       roundingType=HalfHeightRadius
       width=0
       color {colorUseRule=UseColor_Specified, specifiedColor=<rgba 0 0 0 0>}
       cornerRadii=<0 0 0 0>
      }
      tintColor {colorUseRule=UseColor_Specified, specifiedColor=<rgba 0 0.162 0.7454 1>}
      uVRegion {bIsValid=false, max=<0 0>, min=<0 0>}
     }
     undeterminedPressedImage {
      bIsDynamicallyLoaded=false
      drawAs=Image
      imageType=Vector
      mirroring=NoMirror
      resourceName=../../../Engine/Content/Slate/Starship/CoreWidgets/CheckBox/indeterminate.svg
      resourceObject=null
      tiling=NoTile
      imageSize=<25 25>
      margin=<ltrb 0 0 0 0>
      outlineSettings {
       bUseBrushTransparency=false
       roundingType=HalfHeightRadius
       width=0
       color {colorUseRule=UseColor_Specified, specifiedColor=<rgba 0 0 0 0>}
       cornerRadii=<0 0 0 0>
      }
      tintColor {colorUseRule=UseColor_Specified, specifiedColor=<rgba 0 0.0802 0.3515 1>}
      uVRegion {bIsValid=false, max=<0 0>, min=<0 0>}
     }
    }
   }
   slot {class=HorizontalBoxSlot, properties {verticalAlignment=VAlign_Center, padding=<ltrb 10 0 10 0>}}
  }
  VerticalBox_0 {
   widgetClassPath=&VerticalBox
   parent=HorizontalBox_0
   widget=VerticalBox_0
   slot {class=HorizontalBoxSlot, properties {verticalAlignment=VAlign_Center, size {sizeRule=Fill, value=1}}}
  }
  HorizontalBox_157 {
   widgetClassPath=&HorizontalBox
   parent=VerticalBox_0
   widget=HorizontalBox_157
   slot {class=VerticalBoxSlot, properties {size {sizeRule=Fill, value=1}}}
  }
  Image_330 {
   widgetClassPath=&Image
   bIsVariable=true
   parent=HorizontalBox_157
   widget=Image_330
   properties {brush {
     bIsDynamicallyLoaded=false
     drawAs=Image
     imageType=FullColor
     mirroring=NoMirror
     resourceName=None
     resourceObject=/AnimInEdTools/AnimBatchDuplicate/Icons/AnimBatchDup_Icon_Folder
     tiling=NoTile
     imageSize=<16 16>
     margin=<ltrb 0 0 0 0>
     outlineSettings {
      bUseBrushTransparency=false
      roundingType=HalfHeightRadius
      width=0
      color {colorUseRule=UseColor_Specified, specifiedColor=<rgba 0 0 0 0>}
      cornerRadii=<0 0 0 0>
     }
     tintColor {colorUseRule=UseColor_Specified, specifiedColor=<rgba 0.3968 0.2346 0.0782 1>}
     uVRegion {bIsValid=false, max=<0 0>, min=<0 0>}
    }}
   slot {class=HorizontalBoxSlot, properties {horizontalAlignment=HAlign_Left, verticalAlignment=VAlign_Center, padding=<ltrb 2 0 7 0>}}
  }
  SubfolderText {
   widgetClassPath=&TextBlock
   bIsVariable=true
   parent=HorizontalBox_157
   widget=SubfolderText
   properties {
    text=Text Block
    colorAndOpacity {colorUseRule=UseColor_Specified, specifiedColor=<rgba 0.3968 0.2346 0.0782 1>}
    font {
     bForceMonospaced=false
     bMaterialIsStencil=false
     fontMaterial=null
     fontObject=&Roboto
     letterSpacing=0
     monospacedWidth=1
     size=10.5
     skewAmount=0
     typefaceFontName=Bold
     outlineSettings {
      bApplyOutlineToDropShadows=false
      bMiteredCorners=false
      bSeparateFillAlpha=false
      outlineMaterial=null
      outlineSize=0
      outlineColor=<rgba 0 0 0 1>
     }
    }
   }
   slot {class=HorizontalBoxSlot, properties {verticalAlignment=VAlign_Center, size {sizeRule=Fill, value=1}}}
  }
  TargetFolderIcon {
   widgetClassPath=&Image
   bIsVariable=true
   parent=HorizontalBox_157
   widget=TargetFolderIcon
   properties {
    visibility=Collapsed
    brush {
     bIsDynamicallyLoaded=false
     drawAs=Image
     imageType=FullColor
     mirroring=NoMirror
     resourceName=None
     resourceObject=&AnimBatchDup_Icon_Folder
     tiling=NoTile
     imageSize=<16 16>
     margin=<ltrb 0 0 0 0>
     outlineSettings {
      bUseBrushTransparency=false
      roundingType=HalfHeightRadius
      width=0
      color {colorUseRule=UseColor_Specified, specifiedColor=<rgba 0 0 0 0>}
      cornerRadii=<0 0 0 0>
     }
     tintColor {colorUseRule=UseColor_Specified, specifiedColor=<rgba 0.3968 0.2346 0.0782 1>}
     uVRegion {bIsValid=false, max=<0 0>, min=<0 0>}
    }
   }
   slot {class=HorizontalBoxSlot, properties {horizontalAlignment=HAlign_Left, verticalAlignment=VAlign_Center, padding=<ltrb 30 0 7 0>}}
  }
  TargetSubfolderText {
   widgetClassPath=&TextBlock
   bIsVariable=true
   parent=HorizontalBox_157
   widget=TargetSubfolderText
   properties {
    text=Text Block
    visibility=Collapsed
    colorAndOpacity {colorUseRule=UseColor_Specified, specifiedColor=<rgba 0.3968 0.2346 0.0782 1>}
    font {
     bForceMonospaced=false
     bMaterialIsStencil=false
     fontMaterial=null
     fontObject=&Roboto
     letterSpacing=0
     monospacedWidth=1
     size=10.5
     skewAmount=0
     typefaceFontName=Bold
     outlineSettings {
      bApplyOutlineToDropShadows=false
      bMiteredCorners=false
      bSeparateFillAlpha=false
      outlineMaterial=null
      outlineSize=0
      outlineColor=<rgba 0 0 0 1>
     }
    }
   }
   slot {class=HorizontalBoxSlot, properties {verticalAlignment=VAlign_Center, size {sizeRule=Fill, value=1}}}
  }
  SubEntryListView {
   widgetClassPath=/Script/Blutility.EditorUtilityListView
   bIsVariable=true
   parent=VerticalBox_0
   widget=SubEntryListView
   properties {entryWidgetClass=&/EUW_AnimBatchDup_SubEntry.EUW_AnimBatchDup_SubEntry_C}
   slot {class=VerticalBoxSlot}
  }
 ]
}
