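# Editor utility widget that batch-duplicates Level Sequences with their linked AnimSequences and AnimMontages, optionally retargeting to another Animatable Creature.
# Builds selectable per-sequence items from asset links, then duplicates to target paths, reusing existing assets.
# Copies get repaired Level Sequence/AnimSequence links (via inline Python) and remapped montage segments.
# Retargeting runs IK batch retarget, swaps the sequencer spawnable, and bakes the longest animation to Control Rig before relinking.
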EditorUtilityWidgetBlueprint /AnimInEdTools/AnimBatchDuplicate/EUW_AnimBatchDuplicate_v2
graphs [
 Clear Target Paths {nodes [
   Clear Target Paths() {is=title K2Node_FunctionEntry}
   > Set Target LS Path(Target LS Path:String) -> (Output_Get:String)
   > Set Target AS Path(Target AS Path:String) -> (Output_Get:String)
   > Set Target AM Path(Target AM Path:String) -> (Output_Get:String)
  ]}
 Create Items From LevelSequences {nodes [
   Create Items From LevelSequences!() -> (then, Level Sequences:Array of Level Sequence Object References) {is=title K2Node_FunctionEntry}
   > n1 = Sequence()
   > n2 = For Each Loop(Array:Array of Level Sequence Object References=Get Level Sequences() -> (Level Sequences:Array of Level Sequence Object References)) -> (Array Element:Level Sequence Object Reference)
   n3 = GetObjectName(Object=@n2)
   > n4 = Construct Object from Class!(
    execute
    Class=/AnimInEdTools/AnimBatchDuplicate/Data/BP_AnimBatchDup_SubItem_LS.BP_AnimBatchDup_SubItem_LS_C
    self
    Level Sequence:Level Sequence Object Reference=@n2
    Source Name:String=@n3
    Target Name:String=@n3
    Source Full Path:String=Get Object Path String(Object=@n2)
   ) -> (then, ReturnValue:BP Anim Batch Dup Sub Item LS Object Reference)
   > n5 = Construct Object from Class!(
    execute
    Class=/AnimInEdTools/AnimBatchDuplicate/Data/BP_AnimBatchDup_Item.BP_AnimBatchDup_Item_C
    self
    Level Sequence Item:BP Anim Batch Dup Sub Item LS Object Reference=@n4
   ) -> (then, ReturnValue:BP Anim Batch Dup Item Object Reference)
   > Set Local Current Item(Local Current Item:BP Anim Batch Dup Item Object Reference=@n5) -> (Output_Get:BP Anim Batch Dup Item Object Reference)
   > n6 = Sequence!(execute) -> (then_0, then_1, then_2:Exec)
   n7 = Sequence!(execute=@n6.then_1) -> (then_0, then_1)
   > Clear(TargetArray=Get Local Current AnimMontage Paths() -> (Local Current AnimMontage Paths:Array of Strings))
   > n8 = For Each Loop(Array:Array of BP Anim Batch Dup Sub Item AS Object References=Get Anim Sequence Items!(self:BP Anim Batch Dup Item Object Reference=Get Local Current Item() -> (Local Current Item:BP Anim Batch Dup Item Object Reference)) -> (Anim Sequence Items:Array of BP Anim Batch Dup Sub Item AS Object References)) -> (Array Element:BP Anim Batch Dup Sub Item AS Object Reference)
   > n9 = For Each Loop(Array:Array of Strings=Get AnimMontages Paths from AnimSequence Path!(Anim Sequence Path:String=Get Source Full Path!(self:BP Anim Batch Dup Sub Item Object Reference=@n8) -> (Source Full Path:String)) -> (Anim Montage Paths:Array of Strings) {is=title K2Node_CallFunction}) -> (Array Element:String)
   > Add Unique(TargetArray=Get Local Current AnimMontage Paths(), NewItem=@n9)
   n10 = For Each Loop(Exec=@n7.then_1, Array=Get Local Current AnimMontage Paths())
   n11 = GetBaseFilename(InPath=@n10, bRemovePath=true)
   > n12 = Construct Object from Class!(
    execute
    Class=/AnimInEdTools/AnimBatchDuplicate/Data/BP_AnimBatchDup_SubItem_AM.BP_AnimBatchDup_SubItem_AM_C
    self
    Anim Montage:Anim Montage Object Reference
    Source Name:String=@n11
    Target Name:String=@n11
    Source Full Path:String=@n10
   ) -> (then, ReturnValue:BP Anim Batch Dup Sub Item AM Object Reference)
   > Add(
    TargetArray:Array of BP Anim Batch Dup Sub Item AM Object References=Get Anim Montage Items!(self:BP Anim Batch Dup Item Object Reference=Get Local Current Item()) -> (Anim Montage Items:Array of BP Anim Batch Dup Sub Item AM Object References)
    NewItem:"BP Anim Batch Dup Sub Item AM Object Reference (by ref)"=@n12
   ) {is=title K2Node_CallArrayFunction}
   Add(
    execute=@n6.then_2
    TargetArray:Array of BP Anim Batch Dup Item Object References=Get Local Current Items() -> (Local Current Items:Array of BP Anim Batch Dup Item Object References)
    NewItem:"BP Anim Batch Dup Item Object Reference (by ref)"=Get Local Current Item()
   )
   n13 = Get Level Sequence Item!(self:BP Anim Batch Dup Item Object Reference=Get Local Current Item()) -> (Level Sequence Item:BP Anim Batch Dup Sub Item LS Object Reference)
   n14 = For Each Loop(
    Exec=@n6.then_0
    Array=Get Linked AnimSeq Paths from LevelSeq!(Level Sequence:Level Sequence Object Reference=Get Level Sequence!(self:BP Anim Batch Dup Sub Item LS Object Reference=@n13) -> (Level Sequence:Level Sequence Object Reference)) -> (AnimSequences Paths:Array of Strings) {is=title K2Node_CallFunction}
   )
   n15 = GetBaseFilename(InPath=@n14, bRemovePath=true)
   > n16 = Construct Object from Class!(
    execute
    Class=/AnimInEdTools/AnimBatchDuplicate/Data/BP_AnimBatchDup_SubItem_AS.BP_AnimBatchDup_SubItem_AS_C
    self
    Anim Sequence:Anim Sequence Object Reference
    Source Name=@n15
    Target Name=@n15
    Source Full Path=@n14
   ) -> (then, ReturnValue:BP Anim Batch Dup Sub Item AS Object Reference)
   > Add(
    TargetArray:Array of BP Anim Batch Dup Sub Item AS Object References=Get Anim Sequence Items(self=Get Local Current Item())
    NewItem:"BP Anim Batch Dup Sub Item AS Object Reference (by ref)"=@n16
   )
   Return Node!(execute=@n1.then_1, Updated Items:Array of BP Anim Batch Dup Item Object References=Get Local Current Items())
  ]}
 Duplicate Animations {nodes [
   Duplicate Animations!() -> (
    then
    Items:Array of BP Anim Batch Dup Item Object References
    Include AM:bool=~
    Include LS Subfolder:bool=~
    Include AS Subfolder:bool=~
    Include AM Subfolder:bool=~
    Target LS Path:String
    Target AS Path:String
    Target AM Path:String
   ) {is=title K2Node_FunctionEntry}
   > n1 = For Each Loop(Array:Array of BP Anim Batch Dup Item Object References=Get Items() -> (Items:Array of BP Anim Batch Dup Item Object References)) -> (Array Element:BP Anim Batch Dup Item Object Reference)
   > n2 = Set Local Current Item(Local Current Item=@n1)
   > Branch(Condition=Get Selected!(self:BP Anim Batch Dup Item Object Reference=@n2) -> (Selected:bool))
   > n3 = Sequence!(execute) -> (then_0, then_1, then_2:Exec, then_3:Exec, then_4:Exec, then_5:Exec, then_6:Exec, then_7:Exec, then_8:Exec, then_9:Exec)
   > Set Duplicated Level Sequence(Duplicated Level Sequence:Level Sequence Object Reference) -> (Output_Get:Level Sequence Object Reference)
   > Clear(TargetArray:Array of Anim Sequence Object References=Get Duplicated Anim Sequences() -> (Duplicated Anim Sequences:Array of Anim Sequence Object References))
   > Clear(TargetArray:Array of Anim Montage Object References=Get Duplicated Anim Montages() -> (Duplicated Anim Montages:Array of Anim Montage Object References))
   > Clear(TargetArray:Array of Strings=Get Target Anim Sequences Paths() -> (Target Anim Sequences Paths:Array of Strings))
   > Clear!(
    execute
    TargetMap:Map of Anim Sequence Object References to Anim Sequence Object References=Get Source Target Anim Sequences Map() -> (Source Target Anim Sequences Map:Map of Anim Sequence Object References to Anim Sequence Object References)
   ) -> (then) {is=title K2Node_CallFunction}
   > Clear(TargetArray:Array of Level Sequence Anim Sequence Link Item Structures=Get Updated Links() -> (Updated Links:Array of Level Sequence Anim Sequence Link Item Structures)) {is=title K2Node_CallArrayFunction}
   Branch(execute=@n3.then_4, Condition=Is Not Empty!(TargetArray=Get Target Anim Sequences Paths()) -> (ReturnValue:bool))
   > n4 = GetAnimSequenceLinkFromLevelSequence!(
    execute
    InLevelSequence:Level Sequence Object Reference=Get Duplicated Level Sequence() -> (Duplicated Level Sequence:Level Sequence Object Reference)
   ) -> (then, ReturnValue:Level Sequence Anim Sequence Link Object Reference)
   > n5 = For Each Loop(Array:Array of Level Sequence Anim Sequence Link Item Structures=Get AnimSequenceLinks!(self:Level Sequence Anim Sequence Link Object Reference=@n4) -> (AnimSequenceLinks:Array of Level Sequence Anim Sequence Link Item Structures)) -> (Array Element:Level Sequence Anim Sequence Link Item Structure)
   > n6 = Execute Python Script!(
    execute
    PythonScript="import unreal\r\n\r\nsoft_path = unreal.SoftObjectPath(anim_seq_path)\r\nlink.set_editor_property(\"path_to_anim_sequence\", soft_path)\r\n\r\nupd_link = link"
    anim_seq_path:String="Get (a copy)"(Array=Get Target Anim Sequences Paths(), Dimension 1=@n5.Array Index)
    link:Level Sequence Anim Sequence Link Item Structure=@n5
   ) -> (then, ReturnValue, upd_link:"Level Sequence Anim Sequence Link Item Structure (by ref)") {inputs [anim_seq_path, link], outputs [upd_link]}
   > Add(
    TargetArray:Array of Level Sequence Anim Sequence Link Item Structures=Get Updated Links()
    NewItem:"Level Sequence Anim Sequence Link Item Structure (by ref)"=@n6.upd_link
   )
   Set AnimSequenceLinks!(
    execute=@n5.Completed
    AnimSequenceLinks:Array of Level Sequence Anim Sequence Link Item Structures=Get Updated Links()
    self:Level Sequence Anim Sequence Link Object Reference=@n4
   ) -> (then, Output_Get:Array of Level Sequence Anim Sequence Link Item Structures)
   n7 = For Each Loop(Exec=@n3.then_2, Array:Array of BP Anim Batch Dup Sub Item AS Object References=Get Anim Sequence Items(self=Get Local Current Item())) -> (Array Element:BP Anim Batch Dup Sub Item AS Object Reference)
   > Branch(Condition=Get Selected(self:BP Anim Batch Dup Sub Item Object Reference=@n7))
   > n8 = LoadAsset(AssetPath=Get Source Full Path(self=@n7))
   > Is Valid(InputObject=@n8)
   n9 = Make Target Full Path!(
    Target Path:String=Get Target AS Path() -> (Target AS Path:String)
    Include Subfolder:bool=Get Include AS Subfolder() -> (Include AS Subfolder:bool)
    Sub Item:BP Anim Batch Dup Sub Item Object Reference=@n7
   ) -> (Full Path:String, Path Only:String, Name Only:String)
   > n10 = DoesAssetExist(AssetPath=@n9)
   > n11 = Branch(Condition=@n10)
   > n12 = LoadAsset(AssetPath=@n9)
   > n13 = Cast To AnimSequence!(execute, Object=@n12) -> (then, CastFailed, AsAnim Sequence:Anim Sequence Object Reference)
   > Add(TargetArray:Array of Anim Sequence Object References=Get Duplicated Anim Sequences(), NewItem:"Anim Sequence Object Reference (by ref)"=@n13)
   > Add(
    TargetArray:Array of Strings=Get Target Anim Sequences Paths()
    NewItem:"String (by ref)"=GetSoftObjectPath!(Object=@n13) -> (ReturnValue_PathString:String)
   )
   > n14 = Cast To AnimSequence(Object=@n8)
   > Add!(
    execute
    TargetMap:Map of Anim Sequence Object References to Anim Sequence Object References=Get Source Target Anim Sequences Map()
    Key:"Anim Sequence Object Reference (by ref)"=@n14
    Value:"Anim Sequence Object Reference (by ref)"=@n13
   ) -> (then) {is=title K2Node_CallFunction}
   n15 = DuplicateLoadedAsset!(execute=@n11.else, SourceAsset:Object Reference=@n8, DestinationAssetPath:String=@n9) -> (then, ReturnValue:Object Reference)
   > n16 = Cast To AnimSequence(Object=@n15)
   > Add(TargetArray:Array of Anim Sequence Object References=Get Duplicated Anim Sequences(), NewItem:"Anim Sequence Object Reference (by ref)"=@n16) {is=title K2Node_CallArrayFunction}
   > Add(TargetArray:Array of Strings=Get Target Anim Sequences Paths(), NewItem:"String (by ref)"=GetSoftObjectPath(Object=@n16))
   > n17 = Cast To AnimSequence(Object=@n8)
   > Add(TargetMap=Get Source Target Anim Sequences Map(), Key=@n17, Value=@n16) {is=title K2Node_CallFunction}
   n18 = For Each Loop(Exec=@n3.then_5, Array:Array of Anim Montage Object References=Get Duplicated Anim Montages()) -> (Array Element:Anim Montage Object Reference)
   > n19 = For Each Loop(Array:Array of Names=GetMontageSlotNames!(AnimationMontage:Anim Montage Object Reference=@n18) -> (SlotNames:Array of Names)) -> (Array Element:Name)
   > n20 = Get Sequences from Montage!(execute, AnimMontage:Anim Montage Object Reference=@n18, SlotName:Name=@n19) -> (then, ReturnValue:Array of Anim Sequence Base Object References) {is=title K2Node_CallFunction}
   > n21 = For Each Loop(Array:Array of Anim Sequence Base Object References=@n20) -> (Array Element:Anim Sequence Base Object Reference)
   > n22 = Cast To AnimSequence(Object=@n21)
   n23 = Find(
    TargetMap:Map of Anim Sequence Object References to Anim Sequence Object References=Get Source Target Anim Sequences Map()
    Key:"Anim Sequence Object Reference (by ref)"=@n22
   ) -> (Value:Anim Sequence Object Reference)
   > Branch(Condition=@n23.ReturnValue)
   > Set Anim Montage Anim Sequence!(
    execute
    AnimMontage:Anim Montage Object Reference=@n18
    SlotName:Name=@n19
    AnimSequence:Anim Sequence Object Reference=@n23
    SegmentIndex:int=@n21.Array Index
   ) -> (then, ReturnValue:bool) {is=title K2Node_CallFunction}
   SaveLoadedAsset(execute=@n3.then_7, AssetToSave=Get Duplicated Level Sequence(), bOnlyIfIsDirty=false)
   n24 = For Each Loop(Exec=@n3.then_8, Array:Array of Anim Sequence Object References=Get Duplicated Anim Sequences()) -> (Array Element:Anim Sequence Object Reference)
   > SaveLoadedAsset(AssetToSave=@n24, bOnlyIfIsDirty=false)
   n25 = For Each Loop(Exec=@n3.then_9, Array:Array of Anim Montage Object References=Get Duplicated Anim Montages()) -> (Array Element:Anim Montage Object Reference)
   > SaveLoadedAsset(AssetToSave=@n25, bOnlyIfIsDirty=false)
   Branch(execute=@n3.then_3, Condition=Get Include AM() -> (Include AM:bool))
   > n26 = For Each Loop(Array:Array of BP Anim Batch Dup Sub Item AM Object References=Get Anim Montage Items(self=Get Local Current Item())) -> (Array Element:BP Anim Batch Dup Sub Item AM Object Reference)
   > Branch(Condition=Get Selected(self=@n26))
   > n27 = LoadAsset(AssetPath=Get Source Full Path(self=@n26))
   > Is Valid(InputObject=@n27)
   n28 = Make Target Full Path(
    Target Path=Get Target AM Path() -> (Target AM Path:String)
    Include Subfolder=Get Include AM Subfolder() -> (Include AM Subfolder:bool)
    Sub Item=@n26
   )
   > n29 = DoesAssetExist(AssetPath=@n28)
   > n30 = Branch(Condition=@n29)
   > n31 = LoadAsset(AssetPath=@n28)
   > n32 = Cast To AnimMontage!(execute, Object=@n31) -> (then, CastFailed, AsAnim Montage:Anim Montage Object Reference) {is=title K2Node_DynamicCast}
   > Add(TargetArray:Array of Anim Montage Object References=Get Duplicated Anim Montages(), NewItem:"Anim Montage Object Reference (by ref)"=@n32) {is=title K2Node_CallArrayFunction}
   n33 = DuplicateLoadedAsset(execute=@n30.else, SourceAsset=@n27, DestinationAssetPath=@n28)
   > n34 = Cast To AnimMontage(Object=@n33)
   > Add(TargetArray=Get Duplicated Anim Montages(), NewItem=@n34)
   n35 = Get Level Sequence Item(self=Get Local Current Item())
   Branch(execute=@n3.then_1, Condition=Get Selected(self=@n35))
   > n36 = LoadAsset(AssetPath=Get Source Full Path(self=@n35))
   > Is Valid(InputObject=@n36)
   n37 = Make Target Full Path(
    Target Path=Get Target LS Path() -> (Target LS Path:String)
    Include Subfolder=Get Include LS Subfolder() -> (Include LS Subfolder:bool)
    Sub Item=@n35
   )
   > n38 = DoesAssetExist(AssetPath=@n37)
   > n39 = Branch(Condition=@n38)
   > n40 = LoadAsset(AssetPath=@n37)
   > n41 = Cast To LevelSequence!(execute, Object=@n40) -> (then, CastFailed, AsLevel Sequence:Level Sequence Object Reference) {is=title K2Node_DynamicCast}
   > Set Duplicated Level Sequence(Duplicated Level Sequence=@n41)
   n42 = DuplicateLoadedAsset(execute=@n39.else, SourceAsset=@n36, DestinationAssetPath=@n37)
   > n43 = Cast To LevelSequence(Object=@n42)
   > Set Duplicated Level Sequence(Duplicated Level Sequence=@n43)
   n44 = For Each Loop(Exec=@n3.then_6, Array:Array of Anim Sequence Object References=Get Duplicated Anim Sequences()) -> (Array Element:Anim Sequence Object Reference)
   > n45 = GetLevelSequenceLinkFromAnimSequence!(execute, InAnimSequence:Anim Sequence Object Reference=@n44) -> (then, ReturnValue:Anim Sequence Level Sequence Link Object Reference)
   > Set PathToLevelSequence!(
    execute
    PathToLevelSequence:Soft Object Path Structure=GetSoftObjectPath!(Object=Get Duplicated Level Sequence()) -> (ReturnValue:Soft Object Path Structure)
    self:Anim Sequence Level Sequence Link Object Reference=@n45
   ) -> (then, Output_Get:Soft Object Path Structure)
  ]}
 Duplicate Button Visibility {nodes [
   Duplicate Button Visibility() {is=title K2Node_FunctionEntry}
   n1 = AND Boolean(A=Get Using Source Custom Path() -> (Using Source Custom Path:bool), B=Get Using Target Custom Paths() -> (Using Target Custom Paths:bool))
   n2 = AND Boolean(A=Get Using Source Custom Path(), B=IsValid(Object=Get Target AC() -> (Target AC:Animatable Creature Object Reference)))
   n3 = AND Boolean(A=IsValid(Object=Get AC() -> (AC:Animatable Creature Object Reference)), B=Get Using Target Custom Paths())
   n4 = Get AC()
   n5 = Get Target AC()
   n6 = Get Target AC()
   n7 = Get AC()
   n8 = AND Boolean(
    A=AND Boolean(A=IsValid(Object=@n7), B=IsValid(Object=@n6))
    B="Not Equal Exactly (String)"(A=GetObjectName(Object=@n4), B=GetObjectName(Object=@n5))
   )
   > SetVisibility(
    self=Get DuplicateButton() -> (DuplicateButton:Editor Utility Button Object Reference)
    InVisibility=Select(Option 0=Collapsed, Option 1=Visible, Index=OR Boolean!(A=@n8, B=@n3, C:bool=@n2, D:bool=@n1) -> (ReturnValue))
   )
  ]}
 DuplicateAndRetargetAnimationsPerItem {nodes [
   n1 = Sequence!(execute) -> (then_0, then_1)
   n2 = For Each Loop(
    Exec=@n13.then_1
    Array:Array of BP Anim Batch Dup Sub Item AS Object References=Get Anim Sequence Items(self=Get Item() -> (Item:BP Anim Batch Dup Item Object Reference))
   ) -> (Array Element:BP Anim Batch Dup Sub Item AS Object Reference)
   n3 = Make Target Full Path(Target Path=Get Target AS Path(), Include Subfolder=Get Include AS Subfolder(), Sub Item=@n2)
   n4 = DoesAssetExist(execute=@n1.then_0, AssetPath=@n3)
   > Branch(Condition=@n4)
   > SaveAsset!(execute, AssetToSave:String=@n3, bOnlyIfIsDirty:bool=false) -> (then, ReturnValue:bool)
   > CheckoutAsset!(execute, AssetToCheckout:String=@n3) -> (then, ReturnValue:bool)
   n5 = Get Source Full Path(self=@n2)
   n6 = LoadAsset(execute=@n69.then, AssetPath=@n5)
   > n7 = Cast To AnimSequence(Object=@n6)
   n8 = RunBatchRetarget!(
    execute=[@n7.then, @n1.then_1]
    Inputs_AssetsToRetarget:Array of Asset Data Structures=Make Array("[0]":Asset Data Structure=CreateAssetData(InAsset=@n7, bAllowBlueprintClass=false)) -> (Array:Array of Asset Data Structures)
    Inputs_SourceMesh:Skeletal Mesh Object Reference=Get Source Mesh() -> (Source Mesh:Skeletal Mesh Object Reference)
    Inputs_TargetMesh:Skeletal Mesh Object Reference=Get Target Mesh() -> (Target Mesh:Skeletal Mesh Object Reference)
    Inputs_IKRetargetAsset:IKRetargeter Object Reference=Get IK Retargeter() -> (IK Retargeter:IKRetargeter Object Reference)
    Inputs_InOverrideSetNames:Array of Names
    Inputs_Search:String=Get Source Name!(self:BP Anim Batch Dup Sub Item Object Reference=@n2) -> (Source Name:String)
    Inputs_Replace:String=Get Target Name!(self:BP Anim Batch Dup Sub Item Object Reference=@n2) -> (Target Name:String)
    Inputs_Prefix:String
    Inputs_Suffix:String
    Inputs_TargetPath:String=@n3.Path Only
    Inputs_bUseSourcePath:bool=False
    Inputs_bIncludeReferencedAssets:bool=false
    Inputs_bOverwriteExistingFiles:bool=true
    Inputs_bRetainAdditiveFlags:bool=Get Retain Additive Flags() -> (Retain Additive Flags:bool)
   ) -> (then, ReturnValue:Array of Asset Data Structures)
   > n9 = For Each Loop!(Exec, Array:Array of Asset Data Structures=@n8) -> (
    LoopBody
    Array Element_PackageName:Name
    Array Element_PackagePath:Name
    Array Element_AssetName:Name
    Array Element_AssetClassPath:Top Level Asset Path Structure
    Array Index
    Completed
   )
   > n10 = LoadAsset(AssetPath=Append!(A="To String (Name)"(InName=@n9.Array Element_PackagePath), B=/, C:String="To String (Name)"(InName=@n9.Array Element_AssetName)) -> (ReturnValue))
   > n11 = Cast To AnimSequence(Object=@n10)
   > Add(
    TargetArray:Array of Anim Sequence Object References=Get Target AnimSequences() -> (Target AnimSequences:Array of Anim Sequence Object References)
    NewItem:"Anim Sequence Object Reference (by ref)"=@n11
   )
   > SaveLoadedAsset(AssetToSave=@n11, bOnlyIfIsDirty=false)
   n12 = DuplicateAndRetargetAnimationsPerItem!() -> (
    then
    Item:BP Anim Batch Dup Item Object Reference
    IK Retargeter:IKRetargeter Object Reference
    Source AC:Animatable Creature Object Reference
    Target AC:Animatable Creature Object Reference
    Include AM:bool=~
    Target LS Path:String
    Target AS Path:String
    Target AM Path:String
    Include LS Subfolder:bool=~
    Include AS Subfolder:bool=~
    Include AM Subfolder:bool=~
    Find:String
    Replace:String
   ) {is=title K2Node_FunctionEntry}
   > Branch(Condition=Get Selected(self:BP Anim Batch Dup Item Object Reference=@n12))
   > n13 = Sequence!(execute) -> (then_0, then_1, then_2:Exec, then_3:Exec, then_4:Exec)
   > CloseLevelSequence!(execute) -> (then)
   > n14 = Get IK Retarget Controller!(execute, InRetargeterAsset:IKRetargeter Object Reference=Get IK Retargeter()) -> (then, ReturnValue:IKRetargeter Controller Object Reference) {is=title K2Node_CallFunction}
   > Set Source Mesh(Source Mesh:Skeletal Mesh Object Reference=GetPreviewMesh!(self:IKRetargeter Controller Object Reference=@n14, SourceOrTarget:ERetargetSourceOrTarget Enum=Source) -> (ReturnValue:Skeletal Mesh Object Reference)) -> (Output_Get:Skeletal Mesh Object Reference)
   > Set Target Mesh(Target Mesh:Skeletal Mesh Object Reference=GetPreviewMesh(self=@n14, SourceOrTarget=Target)) -> (Output_Get:Skeletal Mesh Object Reference)
   n15 = Get Item()
   > n16 = GetAnimSequenceLinkFromLevelSequence(InLevelSequence=Get Level Sequence(self=Get Level Sequence Item(self=@n15)))
   > Set Source Links(Source Links:Array of Level Sequence Anim Sequence Link Item Structures=Get AnimSequenceLinks(self=@n16)) -> (Output_Get:Array of Level Sequence Anim Sequence Link Item Structures)
   n17 = Get Level Sequence Item(self=Get Item())
   Branch(execute=@n13.then_2, Condition=Get Selected(self:BP Anim Batch Dup Sub Item Object Reference=@n17))
   n18 = Get Source Full Path(self=@n17)
   > n19 = DoesAssetExist(AssetPath=@n18)
   > Branch(Condition=@n19)
   > n20 = LoadAsset(AssetPath=@n18)
   > Is Valid(InputObject=@n20)
   n21 = Make Target Full Path(Target Path=Get Target LS Path(), Include Subfolder=Get Include LS Subfolder(), Sub Item=@n17)
   > n22 = DoesAssetExist(AssetPath=@n21)
   > n23 = Sequence!(execute) -> (then_0, then_1)
   > n24 = Branch(Condition=@n22)
   > n25 = LoadAsset(AssetPath=@n21)
   > n26 = Cast To LevelSequence(Object=@n25)
   > Set Target LS(Target LS:Level Sequence Object Reference=@n26) -> (Output_Get:Level Sequence Object Reference)
   > Set LS Exists(LS Exists:bool=true) -> (Output_Get:bool)
   n27 = DuplicateLoadedAsset(execute=@n24.else, SourceAsset=@n20, DestinationAssetPath=@n21)
   > n28 = Cast To LevelSequence(Object=@n27)
   > Set Target LS(Target LS=@n28)
   > Set LS Exists(LS Exists=false)
   Reroute Node(InputPin=@n23.then_1)
   SaveLoadedAssets(execute=@n13.then_4, AssetsToSave=Get Target AnimSequences(), bOnlyIfIsDirty=false)
   n29 = Get Item()
   Branch(execute=@n13.then_3, Condition=Get Selected(self=Get Level Sequence Item(self=@n29)))
   > n30 = Sequence!(execute) -> (then_0, then_1, then_2:Exec, then_3:Exec, then_4:Exec, then_5:Exec, then_6:Exec, then_7:Exec)
   n31 = For Each Loop!(Exec=@n30.then_1, Array:Array of Anim Sequence Object References=Get Target AnimSequences()) -> (LoopBody, Array Element:Anim Sequence Object Reference, Array Index, Completed)
   n32 = Get Target AnimSET() -> (Target AnimSET:Anim Sequence Object Reference)
   > Branch(Condition="integer > integer"(
     A=GetNumFrames!(AnimationSequenceBase:Anim Sequence Base Object Reference=@n31) -> (NumFrames:int=0)
     B=GetNumFrames(AnimationSequenceBase=@n32) -> (NumFrames=0)
    ))
   > Set Target AnimSET(Target AnimSET:Anim Sequence Object Reference=@n31) -> (Output_Get:Anim Sequence Object Reference)
   OpenLevelSequence!(execute=@n30.then_0, LevelSequence:Level Sequence Object Reference=Get Target LS() -> (Target LS:Level Sequence Object Reference)) -> (then, ReturnValue:bool)
   > n33 = GetSpawnables!(execute, Sequence:Movie Scene Sequence Object Reference=Get Target LS()) -> (then, ReturnValue:Array of Movie Scene Binding Proxy Structures)
   > n34 = For Each Loop(Array:Array of Movie Scene Binding Proxy Structures=@n33) -> (Array Element:Movie Scene Binding Proxy Structure)
   > n35 = GetPossessedObjectClass!(execute, InBinding:"Movie Scene Binding Proxy Structure (by ref)"=@n34) -> (then, ReturnValue:Object Class Reference)
   n36 = Get LS Exists() -> (LS Exists:bool)
   n37 = Get Source AC() -> (Source AC:Animatable Creature Object Reference)
   n38 = Get Target AC()
   > n39 = LoadClassAsset_Blocking(AssetClass=Get Sequencer Actor!(self:Animatable Creature Object Reference=Select(Option 0:Animatable Creature Object Reference=@n37, Option 1:Animatable Creature Object Reference=@n38, Index=@n36) -> (ReturnValue:Animatable Creature Object Reference)) -> (Sequencer Actor:Object Soft Class Reference))
   > Branch(Condition="Equal (Class)"(A=@n39, B=@n35))
   > Remove!(execute, InBinding:"Movie Scene Binding Proxy Structure (by ref)"=@n34) -> (then)
   n40 = LoadClassAsset_Blocking(execute=@n30.then_2, AssetClass=Get Sequencer Actor(self=Get Target AC()))
   > n41 = Construct Object from Class!(execute, Class=@n40, self) -> (then, ReturnValue:Object Reference)
   > n42 = AddSpawnableFromInstance!(
    execute
    self:Level Sequence Editor Subsystem Object Reference=Get LevelSequenceEditorSubsystem() -> (ReturnValue:Level Sequence Editor Subsystem Object Reference) {is=title K2Node_GetEditorSubsystem}
    Sequence:Movie Scene Sequence Object Reference=Get Target LS()
    ObjectToSpawn:Object Reference=@n41
   ) -> (then, ReturnValue:Movie Scene Binding Proxy Structure)
   > n43 = Set Binding(Binding:Movie Scene Binding Proxy Structure=@n42) -> (Output_Get:Movie Scene Binding Proxy Structure)
   > n44 = AddTrack!(
    execute
    InBinding:"Movie Scene Binding Proxy Structure (by ref)"=@n43
    TrackType:Movie Scene Track Class Reference=/Script/MovieSceneTracks.MovieSceneSkeletalAnimationTrack
   ) -> (then, ReturnValue:Movie Scene Track Object Reference)
   > n45 = Set Animation Track(Animation Track:Movie Scene Track Object Reference=@n44) -> (Output_Get:Movie Scene Track Object Reference)
   > n46 = AddSection!(execute, Track:Movie Scene Track Object Reference=@n45) -> (then, ReturnValue:Movie Scene Section Object Reference)
   > n47 = Cast To MovieSceneSkeletalAnimationSection!(execute, Object=@n46) -> (then, CastFailed, AsMovie Scene Skeletal Animation Section:Movie Scene Skeletal Animation Section Object Reference) {is=title K2Node_DynamicCast}
   n48 = Get Target AnimSET()
   > Set Params!(
    execute
    Params:Movie Scene Skeletal Animation Params Structure=Make Movie Scene Skeletal Animation Params!(
     Animation:Anim Sequence Base Object Reference=@n48
     FirstLoopStartFrameOffset:Frame Number Structure=(Value=0)
     StartFrameOffset:Frame Number Structure=(Value=0)
     EndFrameOffset:Frame Number Structure=(Value=0)
     PlayRate:Movie Scene Time Warp Variant Structure="EMovieSceneTimeWarpType::FixedPlayRate(PlayRate=1)"
     bReverse:bool=False
     SlotName:Name=DefaultSlot
     MirrorDataTable:Mirror Data Table Object Reference
     bSkipAnimNotifiers:bool=False
     bForceCustomMode:bool=False
     SwapRootBone:ESwapRootBone Enum=SwapRootBone_None
     bLinearPlaybackWhenScaled:bool=False
    ) -> (MovieSceneSkeletalAnimationParams:Movie Scene Skeletal Animation Params Structure) {is=title K2Node_MakeStruct}
    self:Movie Scene Skeletal Animation Section Object Reference=@n47
   ) -> (then, Output_Get:Movie Scene Skeletal Animation Params Structure)
   > n49 = GetPlaybackRange!(execute, Sequence:Movie Scene Sequence Object Reference=Get Target LS()) -> (then, ReturnValue_bHasStart:bool=~, ReturnValue_bHasEnd:bool=~, ReturnValue_InclusiveStart:int, ReturnValue_ExclusiveEnd:int)
   n50 = GetNumFrames(AnimationSequenceBase=@n48) -> (NumFrames=0)
   > SetRange!(
    execute
    Section:Movie Scene Section Object Reference=@n47
    StartFrame:int=@n49.ReturnValue_InclusiveStart
    EndFrame:int=int + int(A=@n49.ReturnValue_InclusiveStart, B=int + int(A=@n50))
   ) -> (then)
   > n51 = GetEditorWorld(self=Get UnrealEditorSubsystem())
   > n52 = Load Asset Blocking!(
    execute
    InRef:"Control Rig Asset Soft Reference Structure (by ref)"=Get Control Rig!(self:Animatable Creature Object Reference=Get Target AC()) -> (Control Rig:Control Rig Asset Soft Reference Structure)
    bForceLoad:bool=true
   ) -> (then, ReturnValue:Control Rig Asset Strong Reference Structure)
   > n53 = Execute Python Script!(
    execute
    PythonScript="import unreal\r\n## Create animation sequence export options\r\nanim_seq_export_options = unreal.AnimSeqExportOption()\r\nanim_seq_export_options.export_transforms = True\r\nanim_seq_export_options.export_morph_targets = True"
   ) -> (then, ReturnValue, anim_seq_export_options:Anim Seq Export Option Object Reference) {outputs [anim_seq_export_options]}
   > BakeToControlRigFromReference!(
    execute
    World:World Object Reference=@n51
    LevelSequence:Level Sequence Object Reference=Get Target LS()
    InRef:"Control Rig Asset Strong Reference Structure (by ref)"=@n52
    ExportOptions:Anim Seq Export Option Object Reference=@n53.anim_seq_export_options
    bReduceKeys:bool=false
    Tolerance:float=0
    Binding:"Movie Scene Binding Proxy Structure (by ref)"=Get Binding() -> (Binding:Movie Scene Binding Proxy Structure)
    bResetControls:bool=true
   ) -> (then, ReturnValue:bool)
   > RemoveTrack!(
    execute
    InBinding:"Movie Scene Binding Proxy Structure (by ref)"=Get Binding()
    TrackToRemove:Movie Scene Track Object Reference=Get Animation Track() -> (Animation Track:Movie Scene Track Object Reference)
   ) -> (then)
   LinkAnimSequence!(
    execute=@n30.then_3
    Sequence:Level Sequence Object Reference=Get Target LS()
    AnimSequence:Anim Sequence Object Reference=Get Target AnimSET()
    ExportOptions:Anim Seq Export Option Object Reference=Get Anim Export Options() -> (Anim Export Options:Anim Seq Export Option Object Reference)
    Binding:"Movie Scene Binding Proxy Structure (by ref)"=Get Binding()
   ) -> (then, ReturnValue:bool)
   > n54 = GetAssetUserDataOfClass!(
    execute
    self:Interface Asset User Data Interface=Get Target AnimSET()
    InUserDataClass:Asset User Data Class Reference=/Script/LevelSequence.AnimSequenceLevelSequenceLink
   ) -> (then, ReturnValue:Anim Sequence Level Sequence Link Object Reference)
   > Set Skel Track Guid(Skel Track Guid:Guid Structure=Get SkelTrackGuid!(self:Anim Sequence Level Sequence Link Object Reference=@n54) -> (SkelTrackGuid:Guid Structure)) -> (Output_Get:Guid Structure)
   n55 = For Each Loop(Exec=@n30.then_4, Array:Array of Anim Sequence Object References=Get Target AnimSequences()) -> (Array Element:Anim Sequence Object Reference)
   > n56 = GetAssetUserDataOfClass(self=@n55, InUserDataClass=&AnimSequenceLevelSequenceLink)
   > n57 = Sequence!(execute) -> (then_0, then_1)
   > n58 = Is Valid(InputObject=@n56)
   AddAssetUserDataOfClass!(
    execute=@n58.Is Not Valid
    self:Interface Asset User Data Interface=@n55
    InUserDataClass:Asset User Data Class Reference=&AnimSequenceLevelSequenceLink
   ) -> (then, ReturnValue:bool)
   n59 = GetLevelSequenceLinkFromAnimSequence(execute=@n57.then_1, InAnimSequence=@n55)
   > Set PathToLevelSequence(PathToLevelSequence=GetSoftObjectPath(Object=Get Target LS()), self=@n59)
   > Set SkelTrackGuid!(
    execute
    SkelTrackGuid:Guid Structure=Get Skel Track Guid() -> (Skel Track Guid:Guid Structure)
    self:Anim Sequence Level Sequence Link Object Reference=@n59
   ) -> (then, Output_Get:Guid Structure)
   n60 = For Each Loop(Exec=@n30.then_5, Array=Get Target AnimSequences())
   > n61 = For Each Loop(Array:Array of Level Sequence Anim Sequence Link Item Structures=Get Source Links() -> (Source Links:Array of Level Sequence Anim Sequence Link Item Structures)) -> (Array Element:Level Sequence Anim Sequence Link Item Structure)
   n62 = Break Level Sequence Anim Sequence Link Item!(LevelSequenceAnimSequenceLinkItem:"Level Sequence Anim Sequence Link Item Structure (by ref)"=@n61) -> (
    SkelTrackGuid:Guid Structure
    PathToAnimSequence_PathString:String
    bAutoBake:bool=~
    bExportTransforms:bool=~
    bExportMorphTargets:bool=~
    bExportAttributeCurves:bool=~
    bExportMaterialCurves:bool=~
    bSkipCurvesWithZeroValue:bool=~
    bSetRetargetSourceAsset:bool=~
    Interpolation:EAnimInterpolationType Enum
    CurveInterpolation:ERichCurveInterpMode Enum
    bRecordInWorldSpace:bool=~
    bEvaluateAllSkeletalMeshComponents:bool=~
    IncludeAnimationNames:Array of Strings
    ExcludeAnimationNames:Array of Strings
    bRemoveExcludedCurves:bool=~
    WarmUpFrames:Frame Number Structure
    DelayBeforeStart:Frame Number Structure
    bUseCustomTimeRange:bool=~
    CustomStartFrame:Frame Number Structure
    CustomEndFrame:Frame Number Structure
    CustomDisplayRate:Frame Rate Structure
    bUseCustomFrameRate:bool=~
    CustomFrameRate:Frame Rate Structure
   ) {is=title K2Node_BreakStruct}
   n63 = GetObjectName(Object=@n60)
   n64 = Get Replace() -> (Replace:String)
   n65 = Get Find() -> (Find:String)
   > Branch(Condition="Equal Exactly (String)"(
     A=GetBaseFilename(InPath=@n62.PathToAnimSequence_PathString, bRemovePath=true)
     B=Replace(SourceString=@n63, From=@n64, To=@n65, SearchCase=CaseSensitive)
    ))
   > n66 = Execute Python Script!(
    execute
    PythonScript="import unreal\r\n\r\nsoft_path = unreal.SoftObjectPath(anim_seq_path)\r\nlink.set_editor_property(\"path_to_anim_sequence\", soft_path)\r\nlink.set_editor_property(\"skel_track_guid\", skel_guid)\r\n\r\nupd_link = link"
    anim_seq_path:String=Get Object Path String(Object=@n60)
    link:Level Sequence Anim Sequence Link Item Structure=@n61
    skel_guid:Guid Structure=Get Skel Track Guid()
   ) -> (then, ReturnValue, upd_link:"Level Sequence Anim Sequence Link Item Structure (by ref)") {inputs [anim_seq_path, link, skel_guid], outputs [upd_link]}
   > Add(
    TargetArray:Array of Level Sequence Anim Sequence Link Item Structures=Get Target Links() -> (Target Links:Array of Level Sequence Anim Sequence Link Item Structures)
    NewItem:"Level Sequence Anim Sequence Link Item Structure (by ref)"=@n66.upd_link
   )
   n67 = GetAnimSequenceLinkFromLevelSequence(execute=@n30.then_6, InLevelSequence=Get Target LS())
   > Set AnimSequenceLinks(AnimSequenceLinks=Get Target Links(), self=@n67)
   SaveLoadedAsset(execute=@n30.then_7, AssetToSave=Get Target LS(), bOnlyIfIsDirty=false)
   > CloseLevelSequence()
   n68 = DoesAssetExist(execute=@n70.then, AssetPath=@n5)
   > n69 = Branch(Condition=@n68)
   n70 = Branch(execute=@n2.LoopBody, Condition=Get Selected(self=@n2))
  ]}
 DuplicateAnimationsPerItem {nodes [
   n1 = DuplicateAnimationsPerItem!() -> (
    then
    Item:BP Anim Batch Dup Item Object Reference
    Include AM:bool=~
    Include LS Subfolder:bool=~
    Include AS Subfolder:bool=~
    Include AM Subfolder:bool=~
    Target LS Path:String
    Target AS Path:String
    Target AM Path:String
   ) {is=title K2Node_FunctionEntry}
   > n2 = Set Local Current Item(Local Current Item=@n1)
   > Branch(Condition=Get Selected(self:BP Anim Batch Dup Item Object Reference=@n2))
   > n3 = Sequence!(execute) -> (then_0, then_1, then_2:Exec, then_3:Exec, then_4:Exec, then_5:Exec, then_6:Exec, then_7:Exec, then_8:Exec, then_9:Exec)
   > Set Duplicated Level Sequence()
   > Clear(TargetArray:Array of Anim Sequence Object References=Get Duplicated Anim Sequences())
   > Clear(TargetArray:Array of Anim Montage Object References=Get Duplicated Anim Montages())
   > Clear(TargetArray:Array of Strings=Get Target Anim Sequences Paths())
   > Clear(TargetMap=Get Source Target Anim Sequences Map()) {is=title K2Node_CallFunction}
   > Clear(TargetArray:Array of Level Sequence Anim Sequence Link Item Structures=Get Updated Links()) {is=title K2Node_CallArrayFunction}
   Branch(execute=@n3.then_4, Condition=Is Not Empty(TargetArray=Get Target Anim Sequences Paths()))
   > n4 = GetAnimSequenceLinkFromLevelSequence(InLevelSequence=Get Duplicated Level Sequence())
   > n5 = For Each Loop(Array=Get AnimSequenceLinks(self=@n4))
   > n6 = Execute Python Script!(
    execute
    PythonScript="import unreal\r\n\r\nsoft_path = unreal.SoftObjectPath(anim_seq_path)\r\nlink.set_editor_property(\"path_to_anim_sequence\", soft_path)\r\n\r\nupd_link = link"
    anim_seq_path="Get (a copy)"(Array=Get Target Anim Sequences Paths(), Dimension 1=@n5.Array Index)
    link=@n5
   ) -> (then, ReturnValue, upd_link) {inputs [anim_seq_path, link], outputs [upd_link]}
   > Add(TargetArray=Get Updated Links(), NewItem=@n6.upd_link)
   Set AnimSequenceLinks(execute=@n5.Completed, AnimSequenceLinks=Get Updated Links(), self=@n4)
   n7 = For Each Loop(Exec=@n3.then_2, Array:Array of BP Anim Batch Dup Sub Item AS Object References=Get Anim Sequence Items(self=Get Local Current Item())) -> (Array Element:BP Anim Batch Dup Sub Item AS Object Reference)
   > Branch(Condition=Get Selected(self:BP Anim Batch Dup Sub Item Object Reference=@n7))
   > n8 = LoadAsset(AssetPath=Get Source Full Path(self=@n7))
   > Is Valid(InputObject=@n8)
   n9 = Make Target Full Path(Target Path=Get Target AS Path(), Include Subfolder=Get Include AS Subfolder(), Sub Item=@n7)
   > n10 = DoesAssetExist(AssetPath=@n9)
   > n11 = Branch(Condition=@n10)
   > n12 = LoadAsset(AssetPath=@n9)
   > n13 = Cast To AnimSequence(Object=@n12)
   > Add(TargetArray:Array of Anim Sequence Object References=Get Duplicated Anim Sequences(), NewItem:"Anim Sequence Object Reference (by ref)"=@n13)
   > Add(
    TargetArray:Array of Strings=Get Target Anim Sequences Paths()
    NewItem:"String (by ref)"=GetSoftObjectPath!(Object=@n13) -> (ReturnValue_PathString:String)
   )
   > n14 = Cast To AnimSequence(Object=@n8)
   > Add(TargetMap=Get Source Target Anim Sequences Map(), Key=@n14, Value=@n13) {is=title K2Node_CallFunction}
   n15 = DuplicateLoadedAsset(execute=@n11.else, SourceAsset=@n8, DestinationAssetPath=@n9)
   > n16 = Cast To AnimSequence(Object=@n15)
   > Add(TargetArray:Array of Anim Sequence Object References=Get Duplicated Anim Sequences(), NewItem:"Anim Sequence Object Reference (by ref)"=@n16) {is=title K2Node_CallArrayFunction}
   > Add(TargetArray:Array of Strings=Get Target Anim Sequences Paths(), NewItem:"String (by ref)"=GetSoftObjectPath(Object=@n16))
   > n17 = Cast To AnimSequence(Object=@n8)
   > Add(TargetMap=Get Source Target Anim Sequences Map(), Key=@n17, Value=@n16) {is=title K2Node_CallFunction}
   n18 = For Each Loop(Exec=@n3.then_5, Array:Array of Anim Montage Object References=Get Duplicated Anim Montages()) -> (Array Element:Anim Montage Object Reference)
   > n19 = For Each Loop(Array:Array of Names=GetMontageSlotNames(AnimationMontage=@n18)) -> (Array Element:Name)
   > n20 = Get Sequences from Montage(AnimMontage=@n18, SlotName=@n19)
   > n21 = For Each Loop(Array:Array of Anim Sequence Base Object References=@n20) -> (Array Element:Anim Sequence Base Object Reference)
   > n22 = Cast To AnimSequence(Object=@n21)
   n23 = Find(TargetMap=Get Source Target Anim Sequences Map(), Key=@n22)
   > Branch(Condition=@n23.ReturnValue)
   > Set Anim Montage Anim Sequence(AnimMontage=@n18, SlotName=@n19, AnimSequence=@n23, SegmentIndex=@n21.Array Index)
   SaveLoadedAsset(execute=@n3.then_7, AssetToSave=Get Duplicated Level Sequence(), bOnlyIfIsDirty=false)
   n24 = For Each Loop(Exec=@n3.then_8, Array:Array of Anim Sequence Object References=Get Duplicated Anim Sequences()) -> (Array Element:Anim Sequence Object Reference)
   > SaveLoadedAsset(AssetToSave=@n24, bOnlyIfIsDirty=false)
   n25 = For Each Loop(Exec=@n3.then_9, Array:Array of Anim Montage Object References=Get Duplicated Anim Montages()) -> (Array Element:Anim Montage Object Reference)
   > SaveLoadedAsset(AssetToSave=@n25, bOnlyIfIsDirty=false)
   Branch(execute=@n3.then_3, Condition=Get Include AM())
   > n26 = For Each Loop(Array:Array of BP Anim Batch Dup Sub Item AM Object References=Get Anim Montage Items(self=Get Local Current Item())) -> (Array Element:BP Anim Batch Dup Sub Item AM Object Reference)
   > Branch(Condition=Get Selected(self=@n26))
   > n27 = LoadAsset(AssetPath=Get Source Full Path(self=@n26))
   > Is Valid(InputObject=@n27)
   n28 = Make Target Full Path(Target Path=Get Target AM Path(), Include Subfolder=Get Include AM Subfolder(), Sub Item=@n26)
   > n29 = DoesAssetExist(AssetPath=@n28)
   > n30 = Branch(Condition=@n29)
   > n31 = LoadAsset(AssetPath=@n28)
   > n32 = Cast To AnimMontage(Object=@n31)
   > Add(TargetArray:Array of Anim Montage Object References=Get Duplicated Anim Montages(), NewItem:"Anim Montage Object Reference (by ref)"=@n32) {is=title K2Node_CallArrayFunction}
   n33 = DuplicateLoadedAsset(execute=@n30.else, SourceAsset=@n27, DestinationAssetPath=@n28)
   > n34 = Cast To AnimMontage(Object=@n33)
   > Add(TargetArray=Get Duplicated Anim Montages(), NewItem=@n34)
   n35 = Get Level Sequence Item(self=Get Local Current Item())
   Branch(execute=@n3.then_1, Condition=Get Selected(self=@n35))
   > n36 = LoadAsset(AssetPath=Get Source Full Path(self=@n35))
   > Is Valid(InputObject=@n36)
   n37 = Make Target Full Path(Target Path=Get Target LS Path(), Include Subfolder=Get Include LS Subfolder(), Sub Item=@n35)
   > n38 = DoesAssetExist(AssetPath=@n37)
   > n39 = Branch(Condition=@n38)
   > n40 = LoadAsset(AssetPath=@n37)
   > n41 = Cast To LevelSequence(Object=@n40)
   > Set Duplicated Level Sequence(Duplicated Level Sequence=@n41)
   n42 = DuplicateLoadedAsset(execute=@n39.else, SourceAsset=@n36, DestinationAssetPath=@n37)
   > n43 = Cast To LevelSequence(Object=@n42)
   > Set Duplicated Level Sequence(Duplicated Level Sequence=@n43)
   n44 = For Each Loop(Exec=@n3.then_6, Array:Array of Anim Sequence Object References=Get Duplicated Anim Sequences()) -> (Array Element:Anim Sequence Object Reference)
   > n45 = GetLevelSequenceLinkFromAnimSequence(InAnimSequence=@n44)
   > Set PathToLevelSequence(PathToLevelSequence=GetSoftObjectPath!(Object=Get Duplicated Level Sequence()) -> (ReturnValue:Soft Object Path Structure), self=@n45)
  ]}
 EventGraph {nodes [
   "On Property Changed (SourceACPropertyView)"!() -> (OutputDelegate, then, PropertyName:Name=None) {is=title K2Node_ComponentBoundEvent}
   n1 = Get AC()
   > n2 = Create Items From LevelSequences!(
    execute
    self
    Level Sequences:Array of Level Sequence Object References=Get Level Sequences from Path!(
     Path:String=Translate Source Path If Tokenized!(AC:Animatable Creature Object Reference=@n1, Anim Folder:AnimBatchDup_Folder Enum=NewEnumerator0) -> (Path:String)
     Recursive:bool=true
    ) -> (Level Sequences:Array of Level Sequence Object References) {is=title K2Node_CallFunction}
   ) -> (then, Updated Items:Array of BP Anim Batch Dup Item Object References) {is=title K2Node_CallFunction}
   > Set Current List Items(Current List Items:Array of BP Anim Batch Dup Item Object References=@n2) -> (Output_Get:Array of BP Anim Batch Dup Item Object References)
   > Update Items ListView()
   > Duplicate Button Visibility() {is=title K2Node_CallFunction}
   "On Clicked (DuplicateButton)"() {is=title K2Node_ComponentBoundEvent}
   n3 = Now!() -> (
    ReturnValue_Year:int=0
    ReturnValue_Month:int=0
    ReturnValue_Day:int=0
    ReturnValue_Hour:int=0
    ReturnValue_Minute:int=0
    ReturnValue_Second:int=0
    ReturnValue_Millisecond:int=0
   )
   n4 = "To String (Integer)"(InInt=@n3.ReturnValue_Hour)
   n5 = "To String (Integer)"(InInt=@n3.ReturnValue_Minute)
   > PrintString(InString=Append!(A="Duplication Started at: ", B=Append!(A=@n4, B=:, C:String=@n5) -> (ReturnValue)) -> (ReturnValue))
   > n6 = Branch(Condition=Get Bake To New Character() -> (Bake To New Character:bool))
   n7 = Get Target AC()
   > n8 = Branch(Condition=IsValidSoftClassReference!(SoftClassReference:"Object Soft Class Reference (by ref)"=Get Sequencer Actor(self=@n7)) -> (ReturnValue:bool))
   n9 = Get Target AC()
   > n10 = Branch(Condition=Is Valid Reference!(InRef:"Control Rig Asset Soft Reference Structure (by ref)"=Get Control Rig(self=@n9)) -> (ReturnValue:bool))
   > n11 = Is Valid(InputObject=Get IK Retargeter())
   PrintString(execute=@n11.Is Not Valid, InString=IK Retargeter asset is not valid)
   n12 = Sequence!(execute=@n11.Is Valid) -> (then_0, then_1)
   n13 = Get ReplaceLSToEditableText() -> (ReplaceLSToEditableText:Editor Utility Editable Text Box Object Reference)
   n14 = Get ReplaceLSFromEditableText() -> (ReplaceLSFromEditableText:Editor Utility Editable Text Box Object Reference)
   Execute Python Script!(
    execute=@n12.then_1
    PythonScript="import unreal\r\n\r\n## Get a reference of the widget to later call the Duplicate function\r\nwidget: unreal.EditorUtilityWidgetBlueprint = unreal.load_object(None, widget_path)\r\n\r\n##SlowTask\r\nwith unreal.ScopedSlowTask(len(items), text_label) as slow_task:\r\n\r\n    slow_task.make_dialog(True)\r\n\r\n    ##The loop\r\n    for item in items:\r\n\r\n        if slow_task.should_cancel():\r\n            break\r\n        \r\n        slow_task.enter_progress_frame(1)\r\n        \r\n        ##The BP function call, with arguments\r\n        widget.call_method(\r\n            \"DuplicateAndRetargetAnimationsPerItem\",\r\n                args=(\r\n                    item,\r\n                    ik_retargeter,\r\n                    source_ac,\r\n                    target_ac,\r\n                    include_am,\r\n                    target_ls_path,\r\n                    target_as_path,\r\n                    target_am_path,\r\n                    include_ls_subfolder,\r\n                    include_as_subfolder,\r\n                    include_am_subfolder,\r\n                    find,\r\n                    replace\r\n                )\r\n        )"
    items:Array of BP Anim Batch Dup Item Object References=Get Current List Items() -> (Current List Items:Array of BP Anim Batch Dup Item Object References)
    text_label:String=MakeLiteralString(Value=Duplicating Animations)
    include_am:bool=Get Include AnimMontages() -> (Include AnimMontages:bool)
    include_ls_subfolder:bool=Get Include Target LS Subfolder() -> (Include Target LS Subfolder:bool)
    include_as_subfolder:bool=Get Include Target AS Subfolder() -> (Include Target AS Subfolder:bool)
    include_am_subfolder:bool=Get Include Target AM Subfolder() -> (Include Target AM Subfolder:bool)
    target_ls_path:String=Get Target LS Path()
    target_as_path:String=Get Target AS Path()
    target_am_path:String=Get Target AM Path()
    widget_path:String=Get Object Path String(Object=Self-Reference())
    source_ac:Animatable Creature Object Reference=Get AC()
    target_ac:Animatable Creature Object Reference=Get Target AC()
    ik_retargeter:IKRetargeter Object Reference=Get IK Retargeter()
    find:String="To String (Text)"(InText=Get Text(self=@n14))
    replace:String="To String (Text)"(InText=Get Text(self=@n13))
   ) -> (then, ReturnValue) {inputs [
     items
     text_label
     include_am
     include_ls_subfolder
     include_as_subfolder
     include_am_subfolder
     target_ls_path
     target_as_path
     target_am_path
     widget_path
     source_ac
     target_ac
     ik_retargeter
     find
     replace
    ]}
   n15 = Now() -> (
    ReturnValue_Year=0
    ReturnValue_Month=0
    ReturnValue_Day=0
    ReturnValue_Hour=0
    ReturnValue_Minute=0
    ReturnValue_Second=0
    ReturnValue_Millisecond=0
   )
   n16 = "To String (Integer)"(InInt=@n15.ReturnValue_Hour)
   n17 = "To String (Integer)"(InInt=@n15.ReturnValue_Minute)
   > PrintString(InString=Append!(A="Duplication Ended at: ", B=Append!(A=@n16, B=:, C:String=@n17) -> (ReturnValue)) -> (ReturnValue))
   PrintString(execute=@n10.else, InString=Target AC file doesn't contain a valid Control Rig class)
   PrintString(execute=@n8.else, InString=Target AC file doesn't contain a valid Sequence Actor class)
   Execute Python Script!(
    execute=@n6.else
    PythonScript="import unreal\r\n\r\n## Get a reference of the widget to later call the Duplicate function\r\nwidget: unreal.EditorUtilityWidgetBlueprint = unreal.load_object(None, widget_path)\r\n\r\n##SlowTask\r\nwith unreal.ScopedSlowTask(len(items), text_label) as slow_task:\r\n\r\n    slow_task.make_dialog(True)\r\n\r\n    ##The loop\r\n    for item in items:\r\n\r\n        if slow_task.should_cancel():\r\n            break\r\n        \r\n        slow_task.enter_progress_frame(1)\r\n        \r\n        ##The BP function call, with arguments\r\n        widget.call_method(\r\n            \"DuplicateAnimationsPerItem\",\r\n                args=(\r\n                    item, \r\n                    include_am,\r\n                    include_ls_subfolder,\r\n                    include_as_subfolder,\r\n                    include_am_subfolder,\r\n                    target_ls_path,\r\n                    target_as_path,\r\n                    target_am_path\r\n                )\r\n        )"
    items=Get Current List Items()
    text_label=MakeLiteralString(Value=Duplicating Animations)
    include_am=Get Include AnimMontages()
    include_ls_subfolder=Get Include Target LS Subfolder()
    include_as_subfolder=Get Include Target AS Subfolder()
    include_am_subfolder=Get Include Target AM Subfolder()
    target_ls_path=Get Target LS Path()
    target_as_path=Get Target AS Path()
    target_am_path=Get Target AM Path()
    widget_path=Get Object Path String(Object=Self-Reference())
   ) -> (then, ReturnValue) {inputs [
     items
     text_label
     include_am
     include_ls_subfolder
     include_as_subfolder
     include_am_subfolder
     target_ls_path
     target_as_path
     target_am_path
     widget_path
    ]}
   n18 = "On Check State Changed (RenameAnimTypeCheckbox)"!() -> (OutputDelegate, then, bIsChecked:bool) {is=title K2Node_ComponentBoundEvent}
   > n19 = Set Widget Visibility!(
    execute
    self
    Widget:Array of Widget Object References=Make Array!(
     "[0]":Widget Object Reference=Get RenameAnimSequenceHorizontalBox() -> (RenameAnimSequenceHorizontalBox:Horizontal Box Object Reference)
     "[1]":Widget Object Reference=Get RenameAnimMontageHorizontalBox() -> (RenameAnimMontageHorizontalBox:Horizontal Box Object Reference)
    ) -> (Array:Array of Widget Object References)
    Visible:bool=@n18.bIsChecked
   ) -> (then) {is=title K2Node_CallFunction}
   n20 = For Each Loop(
    Exec=[@n19.then, @n97.Completed]
    Array:Array of Text Object References=Make Array(
     "[0]":Text Object Reference=Get FromLSText() -> (FromLSText:Text Object Reference)
     "[1]":Text Object Reference=Get ToLSText() -> (ToLSText:Text Object Reference)
    ) -> (Array:Array of Text Object References)
   ) -> (Array Element:Text Object Reference)
   > SetColorAndOpacity!(
    execute
    self=@n20
    InColorAndOpacity_SpecifiedColor:Linear Color Structure=SelectColor(A=(R=0.5776,G=0.0802,B=0.0802,A=1), B=(R=1,G=1,B=1,A=1), bPickA=@n18.bIsChecked)
    InColorAndOpacity_ColorUseRule:ESlateColorStylingMode Enum
   ) -> (then)
   n21 = "On Text Changed (ReplaceASFromEditableText)"!() -> (OutputDelegate, then, Text:"Text (by ref)") {is=title K2Node_ComponentBoundEvent}
   n22 = Switch on ECheckBoxState!(
    execute=[@n21.then, @n27.then]
    Selection:ECheckBoxState Enum=GetCheckedState!(self:Check Box Object Reference=Get RenameAnimTypeCheckbox() -> (RenameAnimTypeCheckbox:Editor Utility Check Box Object Reference)) -> (ReturnValue:ECheckBoxState Enum=Unchecked)
   ) -> (Unchecked:Exec, Checked:Exec, Undetermined:Exec) {is=title K2Node_SwitchEnum}
   n23 = For Each Loop(Exec=@n22.Checked, Array:Array of BP Anim Batch Dup Item Object References=Get Current List Items()) -> (Array Element:BP Anim Batch Dup Item Object Reference)
   Update Items ListView(execute=@n23.Completed)
   n24 = For Each Loop(Exec=@n23.LoopBody, Array:Array of BP Anim Batch Dup Sub Item AS Object References=Get Anim Sequence Items(self=@n23)) -> (Array Element:BP Anim Batch Dup Sub Item AS Object Reference)
   n25 = Get ReplaceASToEditableText() -> (ReplaceASToEditableText:Editor Utility Editable Text Box Object Reference)
   n26 = Get ReplaceASFromEditableText() -> (ReplaceASFromEditableText:Editor Utility Editable Text Box Object Reference)
   > Replace Name!(
    execute
    self:BP Anim Batch Dup Sub Item Object Reference=@n24
    From:String="To String (Text)"(InText=Get Text(self=@n26))
    To:String="To String (Text)"(InText=Get Text(self=@n25))
   ) -> (then)
   n27 = "On Text Changed (ReplaceASToEditableText)"!() -> (OutputDelegate, then, Text:"Text (by ref)") {is=title K2Node_ComponentBoundEvent}
   n28 = "On Text Changed (ReplaceAMToEditableText)"!() -> (OutputDelegate, then, Text:"Text (by ref)") {is=title K2Node_ComponentBoundEvent}
   n29 = Switch on ECheckBoxState(execute=[@n28.then, @n34.then], Selection=GetCheckedState(self=Get RenameAnimTypeCheckbox()) -> (ReturnValue=Unchecked))
   n30 = For Each Loop(Exec=@n29.Checked, Array:Array of BP Anim Batch Dup Item Object References=Get Current List Items()) -> (Array Element:BP Anim Batch Dup Item Object Reference)
   Update Items ListView(execute=@n30.Completed)
   n31 = For Each Loop(Exec=@n30.LoopBody, Array:Array of BP Anim Batch Dup Sub Item AM Object References=Get Anim Montage Items(self=@n30)) -> (Array Element:BP Anim Batch Dup Sub Item AM Object Reference)
   n32 = Get ReplaceAMToEditableText() -> (ReplaceAMToEditableText:Editor Utility Editable Text Box Object Reference)
   n33 = Get ReplaceAMFromEditableText() -> (ReplaceAMFromEditableText:Editor Utility Editable Text Box Object Reference)
   > Replace Name(self=@n31, From="To String (Text)"(InText=Get Text(self=@n33)), To="To String (Text)"(InText=Get Text(self=@n32)))
   n34 = "On Text Changed (ReplaceAMFromEditableText)"!() -> (OutputDelegate, then, Text:"Text (by ref)") {is=title K2Node_ComponentBoundEvent}
   "On Property Changed (TargetACPropertyView)"!() -> (OutputDelegate, then, PropertyName:Name=None) {is=title K2Node_ComponentBoundEvent}
   > n35 = Sequence!(execute) -> (then_0, then_1, then_2:Exec)
   > Duplicate Button Visibility()
   > Clear Target Paths() {is=title K2Node_CallFunction}
   Set Target LS Path(execute=@n35.then_1, Target LS Path=Translate Source Path If Tokenized(AC=Get Target AC(), Anim Folder=NewEnumerator0))
   > Set Target AS Path(Target AS Path=Translate Source Path If Tokenized(AC=Get Target AC(), Anim Folder=NewEnumerator1))
   n36 = IsEmpty(InString=Translate Source Path If Tokenized(AC=Get Target AC(), Anim Folder=NewEnumerator2))
   > Set Widget Visibility(
    Widget=Make Array(
     "[0]":Widget Object Reference=Get TargetAMPathAsteriscText() -> (TargetAMPathAsteriscText:Text Object Reference)
     "[1]":Widget Object Reference=Get AMPathIsEmptyText() -> (AMPathIsEmptyText:Text Object Reference)
    ) -> (Array:Array of Widget Object References)
    Visible=@n36
   )
   > n37 = Branch(Condition=@n36)
   > Set Target AM Path(Target AM Path=Translate Source Path If Tokenized(AC=Get Target AC(), Anim Folder=NewEnumerator1))
   Set Target AM Path(execute=@n37.else, Target AM Path=Translate Source Path If Tokenized(AC=Get Target AC(), Anim Folder=NewEnumerator2))
   Update Target Path Text Widgets(execute=@n35.then_2)
   n38 = Get Target AC()
   > SetVisibility(
    self=Get TargetPathsfromACAsset() -> (TargetPathsfromACAsset:Vertical Box Object Reference)
    InVisibility=Select(Option 0:ESlateVisibility Enum=Collapsed, Option 1:ESlateVisibility Enum=Visible, Index=IsValid(Object=@n38)) -> (ReturnValue:ESlateVisibility Enum)
   )
   n39 = "On Selection Changed (TargetAnimationsComboBox)"!() -> (OutputDelegate, then, SelectedItem:String, SelectionType:ESelectInfo Enum) {is=title K2Node_ComponentBoundEvent}
   > n40 = Sequence()
   > Set Target AC(Target AC:Animatable Creature Object Reference) -> (Output_Get:Animatable Creature Object Reference)
   > Clear Target Paths()
   > Set Text(self:Text Box Object Reference=Get TargetLSPathEditableTextBox() -> (TargetLSPathEditableTextBox:Editor Utility Editable Text Box Object Reference))
   > Set Text(self=Get TargetASPathEditableTextBox() -> (TargetASPathEditableTextBox:Editor Utility Editable Text Box Object Reference))
   > Set Text(self=Get TargetAMPathEditableTextBox() -> (TargetAMPathEditableTextBox:Editor Utility Editable Text Box Object Reference))
   > SetVisibility(self=Get TargetPathsfromACAsset(), InVisibility=Collapsed)
   n41 = Branch(execute=@n40.then_1, Condition="Equal Exactly (String)"(A=@n39.SelectedItem, B=AC Asset))
   > Set Widget Visibility(
    Widget=Make Array!("[0]"=Get TargetACPropertyViewBox() -> (TargetACPropertyViewBox:Horizontal Box Object Reference)) -> (Array)
    Visible=true
   )
   > Set Widget Visibility(Widget=Make Array("[0]"=Get TargetPathsVerticalBox() -> (TargetPathsVerticalBox:Vertical Box Object Reference)), Visible=false)
   > Set Using Target Custom Paths(Using Target Custom Paths:bool=false) -> (Output_Get:bool)
   Branch(execute=@n41.else, Condition="Equal Exactly (String)"(A=@n39.SelectedItem, B=Paths))
   > Set Widget Visibility(Widget=Make Array("[0]"=Get TargetPathsVerticalBox()), Visible=true)
   > Set Widget Visibility(Widget=Make Array("[0]"=Get TargetACPropertyViewBox()), Visible=false)
   > Set Using Target Custom Paths(Using Target Custom Paths=true)
   Duplicate Button Visibility(execute=@n40.then_2)
   n42 = "On Check State Changed (IncludeAnimSequencesCheckbox)"!() -> (OutputDelegate, then, bIsChecked:bool) {is=title K2Node_ComponentBoundEvent}
   > n43 = Set Include AnimSequences(Include AnimSequences:bool=@n42.bIsChecked) -> (Output_Get:bool)
   > Update Items ListView()
   > Set Widget Visibility(
    Widget=Make Array!(
     "[0]"=Get PathTargetAnimSequencesBox() -> (PathTargetAnimSequencesBox:Horizontal Box Object Reference)
     "[1]":Widget Object Reference=Get ACTargetAnimSequenceBox() -> (ACTargetAnimSequenceBox:Horizontal Box Object Reference)
    ) -> (Array)
    Visible=@n43
   )
   n44 = "On Check State Changed (IncludeLevelSequencesCheckbox)"!() -> (OutputDelegate, then, bIsChecked:bool) {is=title K2Node_ComponentBoundEvent}
   > n45 = Set Include LevelSequences(Include LevelSequences:bool=@n44.bIsChecked) -> (Output_Get:bool)
   > Update Items ListView()
   > Set Widget Visibility(
    Widget=Make Array(
     "[0]"=Get PathTargetLevelSequencesBox() -> (PathTargetLevelSequencesBox:Horizontal Box Object Reference)
     "[1]"=Get ACTargetLevelSequenceBox() -> (ACTargetLevelSequenceBox:Horizontal Box Object Reference)
    )
    Visible=@n45
   )
   n46 = "On Check State Changed (BakeToNewCharacterCheckbox)"!() -> (OutputDelegate, then, bIsChecked:bool) {is=title K2Node_ComponentBoundEvent}
   > n47 = Set Bake To New Character(Bake To New Character:bool=@n46.bIsChecked) -> (Output_Get:bool)
   > Set Widget Visibility(
    Widget=Make Array!("[0]"=Get IKRetargeterPropertyView() -> (IKRetargeterPropertyView:Single Property View Object Reference)) -> (Array)
    Visible=@n47
   )
   n48 = Get IKRetargeterBorder() -> (IKRetargeterBorder:Border Object Reference)
   n49 = Get!(self:Border Object Reference=@n48) -> (
    Background_TintColor:Slate Color Structure
    Background_DrawAs:ESlateBrushDrawType Enum
    Background_Tiling:ESlateBrushTileType Enum
    Background_Mirroring:ESlateBrushMirrorType Enum
    Background_ImageSize:Vector 2D Structure
    Background_Margin:Margin Structure
    Background_ResourceObject:Object Reference
    Background_OutlineSettings:Slate Brush Outline Settings Structure
   )
   > SetBrush!(
    execute
    self:Border Object Reference=@n48
    InBrush_TintColor:Slate Color Structure=Select(
     Option 0:Slate Color Structure=Make Slate Color(SpecifiedColor=(R=0.0194,G=0.0194,B=0.0194,A=1))
     Option 1:Slate Color Structure=Make Slate Color(SpecifiedColor=(R=0.0062,G=0.1006,B=0.3177,A=1))
     Index=@n47
    ) -> (ReturnValue:Slate Color Structure)
    InBrush_DrawAs:ESlateBrushDrawType Enum=@n49.Background_DrawAs
    InBrush_Tiling:ESlateBrushTileType Enum=@n49.Background_Tiling
    InBrush_Mirroring:ESlateBrushMirrorType Enum=@n49.Background_Mirroring
    InBrush_ImageSize:Vector 2D Structure=@n49.Background_ImageSize
    InBrush_Margin:Margin Structure=@n49.Background_Margin
    InBrush_ResourceObject:Object Reference=@n49.Background_ResourceObject
    InBrush_OutlineSettings:Slate Brush Outline Settings Structure=@n49.Background_OutlineSettings
   ) -> (then)
   n50 = "On Check State Changed (SelectAllCheckbox)"!() -> (OutputDelegate, then, bIsChecked:bool) {is=title K2Node_ComponentBoundEvent}
   > n51 = Update All Items Selection!(execute, self, Items:Array of BP Anim Batch Dup Item Object References=Get Current List Items(), Selected:bool=@n50.bIsChecked) -> (then, Updated Items:Array of BP Anim Batch Dup Item Object References)
   > Set Current List Items(Current List Items=@n51)
   > Update Items ListView()
   n52 = "On Text Changed (TargetLSPathEditableTextBox)"!() -> (OutputDelegate, then, Text:"Text (by ref)") {is=title K2Node_ComponentBoundEvent}
   > Set Target LS Path(Target LS Path="To String (Text)"(InText=@n52.Text))
   n53 = "On Text Changed (TargetASPathEditableTextBox)"!() -> (OutputDelegate, then, Text:"Text (by ref)") {is=title K2Node_ComponentBoundEvent}
   > Set Target AS Path(Target AS Path="To String (Text)"(InText=@n53.Text))
   n54 = "On Text Changed (TargetAMPathEditableTextBox)"!() -> (OutputDelegate, then, Text:"Text (by ref)") {is=title K2Node_ComponentBoundEvent}
   > Set Target AM Path(Target AM Path="To String (Text)"(InText=@n54.Text))
   "On Clicked (DocumentationButton)"() {is=title K2Node_ComponentBoundEvent}
   > Execute Python Script!(execute, PythonScript="import webbrowser\r\nwebbrowser.open(url)", url:String=Get Documentation URL() -> (Documentation URL:String)) -> (then, ReturnValue) {inputs [url]}
   n55 = "On Check State Changed (EditorUtilityCheckBox_0)"!() -> (OutputDelegate, then, bIsChecked:bool) {is=title K2Node_ComponentBoundEvent}
   > Set Retain Additive Flags(Retain Additive Flags:bool=@n55.bIsChecked) -> (Output_Get:bool)
   "On Clicked (SubfolderTargetLSPathACButton)"() {is=title K2Node_ComponentBoundEvent}
   > n56 = Set Include Target LS Subfolder(Include Target LS Subfolder:bool=NOT Boolean(A=Get Include Target LS Subfolder())) -> (Output_Get:bool)
   > Set Widget Visibility(Widget=Make Array("[0]"=Get TargetLSSubfolderText() -> (TargetLSSubfolderText:Text Object Reference)), Visible=@n56)
   "On Clicked (SubfolderTargetASPathACButton)"() {is=title K2Node_ComponentBoundEvent}
   > n57 = Set Include Target AS Subfolder(Include Target AS Subfolder:bool=NOT Boolean(A=Get Include Target AS Subfolder())) -> (Output_Get:bool)
   > Set Widget Visibility(Widget=Make Array("[0]"=Get TargetASSubfolderText() -> (TargetASSubfolderText:Text Object Reference)), Visible=@n57)
   "On Clicked (SubfolderTargetAMPathACButton)"() {is=title K2Node_ComponentBoundEvent}
   > n58 = Set Include Target AM Subfolder(Include Target AM Subfolder:bool=NOT Boolean(A=Get Include Target AM Subfolder())) -> (Output_Get:bool)
   > Set Widget Visibility(Widget=Make Array("[0]"=Get TargetAMSubfolderText() -> (TargetAMSubfolderText:Text Object Reference)), Visible=@n58)
   n59 = "On Selection Changed (SourceAnimationsComboBox_1)"!() -> (OutputDelegate, then, SelectedItem:String, SelectionType:ESelectInfo Enum) {is=title K2Node_ComponentBoundEvent}
   > n60 = Sequence()
   > ClearListItems(self=Get ListView() -> (ListView:Editor Utility List View Object Reference))
   > Clear(TargetArray:Array of BP Anim Batch Dup Item Object References=Get Current List Items())
   > Set AC(AC:Animatable Creature Object Reference) -> (Output_Get:Animatable Creature Object Reference)
   > Set Input Source Path(Input Source Path:String) -> (Output_Get:String)
   > Set Text(self=Get InputSourcePathEditableTextBox() -> (InputSourcePathEditableTextBox:Editor Utility Editable Text Box Object Reference))
   n61 = Branch(execute=@n60.then_1, Condition="Equal Exactly (String)"(A=@n59.SelectedItem, B=AC Asset))
   > Set Widget Visibility(Widget=Make Array("[0]"=Get SourceACPropertyViewBox() -> (SourceACPropertyViewBox:Horizontal Box Object Reference)), Visible=true)
   > Set Widget Visibility(Widget=Make Array("[0]"=Get SourcePathPropertyBox() -> (SourcePathPropertyBox:Vertical Box Object Reference)), Visible=false)
   > Set Using Source Custom Path(Using Source Custom Path:bool=false) -> (Output_Get:bool)
   Branch(execute=@n61.else, Condition="Equal Exactly (String)"(A=@n59.SelectedItem, B=Path))
   > Set Widget Visibility(Widget=Make Array("[0]"=Get SourcePathPropertyBox()), Visible=true)
   > Set Widget Visibility(Widget=Make Array("[0]"=Get SourceACPropertyViewBox()), Visible=false)
   > Set Using Source Custom Path(Using Source Custom Path=true)
   Duplicate Button Visibility(execute=@n60.then_2)
   "On Clicked (SubfolderTargetLSPathButton)"() {is=title K2Node_ComponentBoundEvent}
   > n62 = Set Include Target LS Subfolder(Include Target LS Subfolder=NOT Boolean(A=Get Include Target LS Subfolder()))
   > Set Widget Visibility(Widget=Make Array("[0]"=Get TargetLSSubfolderCustomPathText() -> (TargetLSSubfolderCustomPathText:Text Object Reference)), Visible=@n62)
   "On Clicked (SubfolderTargetASPathButton)"() {is=title K2Node_ComponentBoundEvent}
   > n63 = Set Include Target AS Subfolder(Include Target AS Subfolder=NOT Boolean(A=Get Include Target AS Subfolder()))
   > Set Widget Visibility(Widget=Make Array("[0]"=Get TargetASSubfolderCustomPathText() -> (TargetASSubfolderCustomPathText:Text Object Reference)), Visible=@n63)
   "On Clicked (SubfolderTargetAMPathButton)"() {is=title K2Node_ComponentBoundEvent}
   > n64 = Set Include Target AM Subfolder(Include Target AM Subfolder=NOT Boolean(A=Get Include Target AM Subfolder()))
   > Set Widget Visibility(Widget=Make Array("[0]"=Get TargetAMSubfolderCustomPathText() -> (TargetAMSubfolderCustomPathText:Text Object Reference)), Visible=@n64)
   n65 = "On Entry Initialized (ListView)"!() -> (OutputDelegate, then, Item:Object Reference, Widget:User Widget Object Reference) {is=title K2Node_ComponentBoundEvent}
   > n66 = Cast To EUW_AnimBatchDup_Entry!(execute, Object=@n65.Widget) -> (then, CastFailed, AsEUW Anim Batch Dup Entry:EUW Anim Batch Dup Entry Object Reference) {is=title K2Node_DynamicCast}
   > Populate SubItems ListView!(
    execute
    self:EUW Anim Batch Dup Entry Object Reference=@n66
    Include AM:bool=Get Include AnimMontages()
    Include AS:bool=Get Include AnimSequences() -> (Include AnimSequences:bool)
    Include LS:bool=Get Include LevelSequences() -> (Include LevelSequences:bool)
   ) -> (then)
   n67 = "On Text Changed (ReplaceLSFromEditableText)"!() -> (OutputDelegate, then, Text:"Text (by ref)") {is=title K2Node_ComponentBoundEvent}
   n68 = For Each Loop(Exec=[@n67.then, @n79.then], Array:Array of BP Anim Batch Dup Item Object References=Get Current List Items()) -> (Array Element:BP Anim Batch Dup Item Object Reference)
   > n69 = Sequence!(execute) -> (then_0, then_1)
   n70 = Get ReplaceLSToEditableText()
   n71 = Get ReplaceLSFromEditableText()
   > Replace Name(
    self=Get Level Sequence Item(self=@n68)
    From="To String (Text)"(InText=Get Text(self=@n71))
    To="To String (Text)"(InText=Get Text(self=@n70))
   )
   Switch on ECheckBoxState(execute=@n69.then_1, Selection=GetCheckedState(self=Get RenameAnimTypeCheckbox()) -> (ReturnValue=Unchecked))
   > n72 = Sequence()
   > n73 = For Each Loop(Array:Array of BP Anim Batch Dup Sub Item AS Object References=Get Anim Sequence Items(self=@n68)) -> (Array Element:BP Anim Batch Dup Sub Item AS Object Reference)
   n74 = Get ReplaceLSToEditableText()
   n75 = Get ReplaceLSFromEditableText()
   > Replace Name(self=@n73, From="To String (Text)"(InText=Get Text(self=@n75)), To="To String (Text)"(InText=Get Text(self=@n74)))
   n76 = For Each Loop(Exec=@n72.then_1, Array:Array of BP Anim Batch Dup Sub Item AM Object References=Get Anim Montage Items(self=@n68)) -> (Array Element:BP Anim Batch Dup Sub Item AM Object Reference)
   n77 = Get ReplaceLSToEditableText()
   n78 = Get ReplaceLSFromEditableText()
   > Replace Name(self=@n76, From="To String (Text)"(InText=Get Text(self=@n78)), To="To String (Text)"(InText=Get Text(self=@n77)))
   Update Items ListView(execute=@n68.Completed)
   n79 = "On Text Changed (ReplaceLSToEditableText)"!() -> (OutputDelegate, then, Text:"Text (by ref)") {is=title K2Node_ComponentBoundEvent}
   n80 = "On Check State Changed (InputPathRecursiveCheckbox)"!() -> (OutputDelegate, then, bIsChecked:bool) {is=title K2Node_ComponentBoundEvent}
   > n81 = Set Input Source Path Recursive(Input Source Path Recursive:bool=@n80.bIsChecked) -> (Output_Get:bool)
   > n82 = Create Items From LevelSequences(Level Sequences=Get Level Sequences from Path(Path=Get Input Source Path() -> (Input Source Path:String), Recursive=@n81))
   > Set Current List Items(Current List Items=@n82)
   > Update Items ListView()
   > Duplicate Button Visibility()
   n83 = "On Text Committed (InputSourcePathEditableTextBox)"!() -> (OutputDelegate, then, Text:"Text (by ref)", CommitMethod:ETextCommit Enum=Default) {is=title K2Node_ComponentBoundEvent}
   > n84 = Set Input Source Path(Input Source Path="To String (Text)"(InText=@n83.Text))
   > Branch(Condition=ValidatePath!(InPath:String=@n84) -> (bDidSucceed:bool, OutReason:Text))
   > n85 = Create Items From LevelSequences(Level Sequences=Get Level Sequences from Path(Path=@n84, Recursive=Get Input Source Path Recursive() -> (Input Source Path Recursive:bool)))
   > Set Current List Items(Current List Items=@n85)
   > Update Items ListView()
   > Duplicate Button Visibility()
   n86 = "On Check State Changed (IncludeAnimMontagesCheckbox)"!() -> (OutputDelegate, then, bIsChecked:bool) {is=title K2Node_ComponentBoundEvent}
   > n87 = Set Include AnimMontages(Include AnimMontages:bool=@n86.bIsChecked) -> (Output_Get:bool)
   > Update Items ListView()
   > Set Widget Visibility(
    Widget=Make Array!(
     "[0]"=Get PathTargetAnimMontagesBox() -> (PathTargetAnimMontagesBox:Horizontal Box Object Reference)
     "[1]":Widget Object Reference=Get ACTargetAnimMontagesBox() -> (ACTargetAnimMontagesBox:Horizontal Box Object Reference)
    ) -> (Array)
    Visible=@n87
   )
   Update Items ListView() {is=title K2Node_CustomEvent}
   > ClearListItems(self=Get ListView())
   > RegenerateAllEntries(self:List View Base Object Reference=Get ListView())
   > Set Current List Items(Current List Items=Sort Current Items Alphabetically!(Items:Array of BP Anim Batch Dup Item Object References=Get Current List Items()) -> (Sorted Items:Array of BP Anim Batch Dup Item Object References))
   > Set List Items!(execute, self:List View Object Reference=Get ListView(), InListItems:Array of Object References=Get Current List Items()) -> (then) {is=title K2Node_CallFunction}
   ClearLog() {is=title K2Node_CustomEvent}
   > Set Log(Log:String) -> (Output_Get:String)
   > UpdateLog()
   UpdateLog() {is=title K2Node_CustomEvent}
   > Set Text(Text="To Text (String)"(InString=Get Log() -> (Log:String)), self:Text Object Reference=Get LogText() -> (LogText:Text Object Reference))
   Event PreConstruct()
   > SetObject(self=Get SourceACPropertyView() -> (SourceACPropertyView:Single Property View Object Reference), NewObject=Self-Reference())
   > SetObject(self=Get TargetACPropertyView() -> (TargetACPropertyView:Single Property View Object Reference), NewObject=Self-Reference())
   > SetObject(self=Get IKRetargeterPropertyView(), NewObject=Self-Reference())
   > SetVisibility(self=Get LogBorder() -> (LogBorder:Border Object Reference), InVisibility=Collapsed)
   Event Construct()
   > Set Widget Visibility(
    Widget=Make Array!(
     "[0]"=Get SourcePathPropertyBox()
     "[1]"=Get RenameAnimSequenceHorizontalBox()
     "[2]":Widget Object Reference=Get RenameAnimMontageHorizontalBox()
     "[3]":Widget Object Reference=Get DuplicateButton()
     "[4]":Widget Object Reference=Get TargetPathsVerticalBox()
     "[5]":Widget Object Reference=Get TargetPathsfromACAsset()
     "[6]":Widget Object Reference=Get TargetAMPathAsteriscText()
     "[7]":Widget Object Reference=Get AMPathIsEmptyText()
     "[8]":Widget Object Reference=Get IKRetargeterPropertyView()
    ) -> (Array)
    Visible=false
   )
   > n88 = Execute Python Script!(
    execute
    PythonScript="import unreal\r\n## Create animation sequence export options\r\nanim_seq_export_options = unreal.AnimSeqExportOption()\r\nanim_seq_export_options.export_transforms = True\r\nanim_seq_export_options.export_morph_targets = True"
   ) -> (then, ReturnValue, anim_seq_export_options:Anim Seq Export Option Object Reference) {outputs [anim_seq_export_options]}
   > Set Anim Export Options(Anim Export Options:Anim Seq Export Option Object Reference=@n88.anim_seq_export_options) -> (Output_Get:Anim Seq Export Option Object Reference)
   Reroute Node()
   > n89 = For Each Loop(Array:Array of BP Anim Batch Dup Item Object References=Get Current List Items()) -> (Array Element:BP Anim Batch Dup Item Object Reference)
   > n90 = Sequence()
   n91 = Get Level Sequence Item(self=@n89)
   > Branch(Condition=Get Selected(self=@n91))
   n92 = Make Target Full Path(Target Path=Get Target LS Path(), Include Subfolder=Get Include Target LS Subfolder(), Sub Item=@n91)
   > n93 = DoesAssetExist(AssetPath=@n92)
   > Branch(Condition=@n93)
   > CheckoutAsset(AssetToCheckout=@n92)
   n94 = For Each Loop(Exec=@n90.then_1, Array:Array of BP Anim Batch Dup Sub Item AS Object References=Get Anim Sequence Items(self=@n89)) -> (Array Element:BP Anim Batch Dup Sub Item AS Object Reference)
   > Branch(Condition=Get Selected(self=@n94))
   n95 = Make Target Full Path(Target Path=Get Target AS Path(), Include Subfolder=Get Include Target AS Subfolder(), Sub Item=@n94)
   > n96 = DoesAssetExist(AssetPath=@n95)
   > Branch(Condition=@n96)
   > CheckoutAsset(AssetToCheckout=@n95)
   n97 = For Each Loop(Array:Array of Editor Utility Editable Text Box Object References=Make Array!(
     "[0]":Editor Utility Editable Text Box Object Reference=Get ReplaceASToEditableText()
     "[1]":Editor Utility Editable Text Box Object Reference=Get ReplaceASFromEditableText()
     "[2]":Editor Utility Editable Text Box Object Reference=Get ReplaceAMToEditableText()
     "[3]":Editor Utility Editable Text Box Object Reference=Get ReplaceAMFromEditableText()
    ) -> (Array:Array of Editor Utility Editable Text Box Object References)) -> (Array Element:Editor Utility Editable Text Box Object Reference)
   > Set Text(self:Text Box Object Reference=@n97)
  ]}
 Get AnimMontages Paths from AnimSequence Path {nodes [
   Get AnimMontages Paths from AnimSequence Path!() -> (then, Anim Sequence Path:String) {is=title K2Node_FunctionEntry}
   > n1 = Sequence()
   > n2 = FindAssetData!(execute, AssetPath:String=Get Anim Sequence Path() -> (Anim Sequence Path:String)) -> (then, ReturnValue_PackageName:Name, ReturnValue_PackagePath:Name, ReturnValue_AssetName:Name, ReturnValue_AssetClassPath_PathString:String)
   > n3 = Get Referencers!(
    execute
    self:Asset Registry Interface=GetAssetRegistry()
    PackageName:Name=@n2
    ReferenceOptions:"Asset Registry Dependency Options Structure (by ref)"=Make Asset Registry Dependency Options(bIncludeSoftPackageReferences=True)
   ) -> (then, OutReferencers:Array of Names, ReturnValue:bool) {is=title K2Node_CallFunction}
   > n4 = For Each Loop(Array:Array of Names=@n3) -> (Array Element:Name)
   > n5 = FindAssetData(AssetPath="To String (Name)"(InName=@n4))
   > Branch(Condition=Contains(SearchIn=@n5.ReturnValue_AssetClassPath_PathString, Substring=Engine.AnimMontage))
   > Add Unique(
    TargetArray=Get Local Anim Montage Names() -> (Local Anim Montage Names:Array of Strings)
    NewItem=Append(A="To String (Name)"(InName=@n5), B=., C="To String (Name)"(InName=@n5.ReturnValue_AssetName))
   )
   Return Node!(execute=@n1.then_1, Anim Montage Paths:Array of Strings=Get Local Anim Montage Names())
  ]}
 Get Level Sequences from Path {nodes [
   Get Level Sequences from Path!() -> (then, Path:String, Recursive:bool=true) {is=title K2Node_FunctionEntry}
   > n1 = Sequence()
   Return Node!(
    execute=@n1.then_1
    Level Sequences:Array of Level Sequence Object References=Get Local Level Sequences() -> (Local Level Sequences:Array of Level Sequence Object References)
   )
   > n2 = GetAssetsByPath!(
    execute
    self:Asset Registry Interface=GetAssetRegistry()
    PackagePath:Name=String To Name(InString=Get Path() -> (Path:String))
    bRecursive:bool=Get Recursive() -> (Recursive:bool)
    bIncludeOnlyOnDiskAssets:bool=false
   ) -> (then, OutAssetData:Array of Asset Data Structures, ReturnValue:bool)
   > n3 = For Each Loop!(Exec, Array:Array of Asset Data Structures=@n2) -> (
    LoopBody
    Array Element_PackageName:Name
    Array Element_PackagePath:Name
    Array Element_AssetName:Name
    Array Element_AssetClassPath_PathString:String
    Array Index
    Completed
   )
   > Branch(Condition="Equal Exactly (String)"(A=@n3.Array Element_AssetClassPath_PathString, B=/Script/LevelSequence))
   > n4 = LoadAsset(AssetPath="To String (Name)"(InName=@n3))
   > n5 = Cast To LevelSequence(Object=@n4)
   > Add(TargetArray:Array of Level Sequence Object References=Get Local Level Sequences(), NewItem:"Level Sequence Object Reference (by ref)"=@n5)
  ]}
 Get Linked AnimSeq Paths from LevelSeq {nodes [
   n1 = Get Linked AnimSeq Paths from LevelSeq!() -> (then, Level Sequence:Level Sequence Object Reference) {is=title K2Node_FunctionEntry}
   > n2 = Sequence()
   > n3 = GetAnimSequenceLinkFromLevelSequence(InLevelSequence=@n1)
   > n4 = For Each Loop!(Exec, Array:Array of Level Sequence Anim Sequence Link Item Structures=Get AnimSequenceLinks(self=@n3)) -> (LoopBody, Array Element:Level Sequence Anim Sequence Link Item Structure, Array Index, Completed)
   n5 = Break Level Sequence Anim Sequence Link Item(LevelSequenceAnimSequenceLinkItem=@n4) -> (
    bAutoBake=~
    bExportTransforms=~
    bExportMorphTargets=~
    bExportAttributeCurves=~
    bExportMaterialCurves=~
    bSkipCurvesWithZeroValue=~
    bSetRetargetSourceAsset=~
    bRecordInWorldSpace=~
    bEvaluateAllSkeletalMeshComponents=~
    bRemoveExcludedCurves=~
    bUseCustomTimeRange=~
    bUseCustomFrameRate=~
   )
   > n6 = DoesAssetExist(AssetPath=@n5.PathToAnimSequence_PathString)
   > Branch(Condition=@n6)
   > Add(
    TargetArray:Array of Strings=Get Local AnimSequences Names() -> (Local AnimSequences Names:Array of Strings)
    NewItem:"String (by ref)"=@n5.PathToAnimSequence_PathString
   )
   Return Node!(execute=@n2.then_1, AnimSequences Paths:Array of Strings=Get Local AnimSequences Names())
  ]}
 Get Subfolder from Full Path {nodes [
   Get Subfolder from Full Path!() -> (then, Full Path:String, Name:String, AC:Animatable Creature Object Reference) {is=title K2Node_FunctionEntry}
   > n1 = Sequence()
   Return Node!(
    execute=@n1.then_1
    Subfolder:String=Get Local Subfolder() -> (Local Subfolder:String)
    Matches Existing Subfolder:bool=Get Matches Project Subfolders() -> (Matches Project Subfolders:bool)
    Matching Subfolder1:String=Get Matching Subfolder() -> (Matching Subfolder:String)
   )
   n2 = Replace(
    SourceString=Split(SourceString=Get Full Path() -> (Full Path:String), InStr=., SearchCase=CaseSensitive, SearchDir=FromStart)
    From=Get Name() -> (Name:String)
    SearchCase=CaseSensitive
   )
   n3 = Get Source Path!(self:Animatable Creature Object Reference=Get AC()) -> (Source Path:String)
   > Set Local Subfolder(Local Subfolder:String=Replace(SourceString=Replace(SourceString=@n2, From=@n3, SearchCase=CaseSensitive), From=/, SearchCase=CaseSensitive)) -> (Output_Get:String)
   > n4 = For Each Loop(Array:Array of Strings=Get Subfolders!(self:Animatable Creature Object Reference=Get AC()) -> (Subfolders:Array of Strings) {is=title K2Node_CallFunction}) -> (Array Element:String)
   n5 = For Each Loop(
    Exec=@n4.Completed
    Array:Array of BP Animation Browser Project Object References=Get Projects!(self:Animatable Creature Object Reference=Get AC()) -> (Projects:Array of BP Animation Browser Project Object References) {is=title K2Node_CallFunction}
   ) -> (Array Element:BP Animation Browser Project Object Reference)
   n6 = For Each Loop with Break!(Exec:Exec=@n5.Completed, Array:Array of Strings=Get Project Subfolders() -> (Project Subfolders:Array of Strings), Break:Exec=@n7.then) -> (LoopBody:Exec, Array Element:String, Array Index:int, Completed:Exec)
   > Branch(Condition="Equal Exactly (String)"(A=@n6, B=Get Local Subfolder()))
   > Set Matching Subfolder(Matching Subfolder:String=@n6) -> (Output_Get:String)
   > n7 = Set Matches Project Subfolders(Matches Project Subfolders:bool=true) -> (Output_Get:bool)
   n8 = For Each Loop(Exec=@n5.LoopBody, Array:Array of Strings=Get Subfolders(self:BP Animation Browser Project Object Reference=@n5)) -> (Array Element:String)
   > Add Unique(TargetArray=Get Project Subfolders(), NewItem=@n8)
   Add Unique(execute=@n4.LoopBody, TargetArray=Get Project Subfolders(), NewItem=@n4)
  ]}
 Make Target Full Path {nodes [
   n1 = Make Target Full Path!() -> (then, Target Path:String, Include Subfolder:bool=~, Sub Item:BP Anim Batch Dup Sub Item Object Reference) {is=title K2Node_FunctionEntry}
   n2 = SelectString(
    A=Append!(
     A=@n1
     B=/
     C=Get Subfolder!(self:BP Anim Batch Dup Sub Item Object Reference=@n1.Sub Item) -> (Subfolder:String) {is=title K2Node_CallFunction}
     D:String=/
     E:String=Get Target Name(self=@n1.Sub Item)
    ) -> (ReturnValue)
    B=Append!(A=@n1, B=/, C=Get Target Name(self=@n1.Sub Item)) -> (ReturnValue)
    bPickA=@n1.Include Subfolder
   )
   n3 = Split(SourceString=@n2, InStr=/, SearchDir=FromEnd)
   > Return Node!(execute, Full Path:String=@n2, Path Only:String=@n3, Name Only:String=@n3.RightS)
  ]}
 Set Widget Visibility {nodes [
   n1 = Set Widget Visibility!() -> (then, Widget:Array of Widget Object References, Visible:bool=true) {is=title K2Node_FunctionEntry}
   > n2 = For Each Loop(Array:Array of Widget Object References=@n1) -> (Array Element:Widget Object Reference)
   > SetVisibility(
    self=@n2
    InVisibility=Select(Option 0:ESlateVisibility Enum=Collapsed, Option 1:ESlateVisibility Enum=Visible, Index=Get Visible() -> (Visible:bool)) -> (ReturnValue:ESlateVisibility Enum)
   )
  ]}
 Sort Current Items Alphabetically {nodes [
   Sort Current Items Alphabetically!() -> (then, Items:Array of BP Anim Batch Dup Item Object References) {is=title K2Node_FunctionEntry}
   > n1 = Sequence()
   Return Node!(execute=@n1.then_1, Sorted Items:Array of BP Anim Batch Dup Item Object References=Get Items())
   > n2 = For Each Loop(Array:Array of BP Anim Batch Dup Item Object References=Get Items()) -> (Array Element:BP Anim Batch Dup Item Object Reference)
   > n3 = For Loop(
    FirstIndex=@n2.Array Index
    LastIndex=Last Index!(TargetArray:Array of BP Anim Batch Dup Item Object References=Get Items()) -> (ReturnValue=0)
   )
   n4 = Get Items()
   n5 = Get Source Name(self=Get Level Sequence Item(self="Get (a copy)"(Array:Array of BP Anim Batch Dup Item Object References=@n4, Dimension 1=@n3) -> (Output:BP Anim Batch Dup Item Object Reference)))
   n6 = Len(S=@n5)
   n7 = Get Source Name(self=Get Level Sequence Item(self=@n2))
   n8 = Len(S=@n7)
   n9 = "integer > integer"(A=@n6, B=@n8)
   > n10 = For Loop with Break!(
    Execute:Exec
    FirstIndex:int
    LastIndex:int=int - int(A=Select(Option 0:int=@n6, Option 1:int=@n8, Index=@n9) -> (ReturnValue:int))
    Break:Exec=@n13.then
   ) -> (LoopBody:Exec, Index:int, Completed:Exec) {is=title K2Node_MacroInstance}
   n11 = GetCharacterAsNumber!(SourceString:String=@n7, Index:int=@n10) -> (ReturnValue:int=0)
   n12 = GetCharacterAsNumber(SourceString=@n5, Index=@n10) -> (ReturnValue=0)
   > n13 = Branch(Condition="Not Equal (Integer)"(A=@n12, B=@n11))
   Branch(execute=@n10.Completed, Condition="integer < integer"(A=@n12, B=@n11))
   > Swap Array Elements!(execute, TargetArray:Array of BP Anim Batch Dup Item Object References=Get Items(), FirstIndex:int=@n3, SecondIndex:int=@n2.Array Index) -> (then) {is=title K2Node_CallArrayFunction}
   "integer > integer"(A:Wildcard, B:Wildcard)
  ]}
 Translate Source Path If Tokenized {nodes [
   Translate Source Path If Tokenized!() -> (then, AC:Animatable Creature Object Reference, Anim Folder:AnimBatchDup_Folder Enum=NewEnumerator0) {is=title K2Node_FunctionEntry}
   > n1 = Sequence!(execute) -> (then_0, then_1, then_2:Exec)
   Return Node!(execute=@n1.then_2, Path:String=Get Final Path() -> (Final Path:String))
   > n2 = Switch on AnimBatchDup_Folder!(execute, Selection:AnimBatchDup_Folder Enum=Get Anim Folder() -> (Anim Folder:AnimBatchDup_Folder Enum=NewEnumerator0)) -> (NewEnumerator0:Exec, NewEnumerator1:Exec, NewEnumerator2:Exec) {is=title K2Node_SwitchEnum}
   > Set Initial Path(Initial Path:String=Get Source Path(self=Get AC())) -> (Output_Get:String)
   Set Initial Path(execute=@n2.NewEnumerator1, Initial Path=Get Animation Path!(self:Animatable Creature Object Reference=Get AC()) -> (Animation Path:String))
   Set Initial Path(
    execute=@n2.NewEnumerator2
    Initial Path=Get AnimMontage Path!(self:Animatable Creature Object Reference=Get AC()) -> (AnimMontage Path:String)
   )
   n3 = Get Initial Path() -> (Initial Path:String)
   n4 = Branch(execute=@n1.then_1, Condition=OR Boolean!(A=Contains(SearchIn=@n3, Substring="{"), B=Contains(SearchIn=@n3, Substring="}")) -> (ReturnValue))
   n5 = Get AC()
   n6 = Get AC()
   n7 = Get Naming Convention!(self:Animatable Creature Object Reference=@n6) -> (ReturnValue:EU Animation Naming Convention Object Reference) {is=title K2Node_CallFunction}
   n8 = Get Asset!(self:Animatable Creature Object Reference=@n6) -> (Asset:String)
   n9 = Get Variant!(self:Animatable Creature Object Reference=@n6) -> (Variant:String)
   > Set Final Path(Final Path:String=Get Path!(
     self:EU Animation Naming Convention Object Reference=Get Naming Convention(self=@n5)
     Token Map:Map of Strings to Strings=Get Path Parts Map!(
      Asset:String=@n8
      Animation Set:String
      Animation Name:String
      Descriptor:String
      Direction:String
      Subfolder:String
      Variation:String
      Plugin:String=@n9
      __WorldContext:Object Reference=@n7
     ) -> (Output Path Parts Map:Map of Strings to Strings) {is=title K2Node_CallFunction}
     Path Pattern:String=Get Initial Path()
     Remove Spaces On Name:bool=true
    ) -> (File Path:String, Folder Path:String, File Name:String) {is=title K2Node_CallFunction}.Folder Path) -> (Output_Get:String)
   Set Final Path(execute=@n4.else, Final Path=Get Initial Path())
  ]}
 Update All Items Selection {nodes [
   Update All Items Selection!() -> (then, Items:Array of BP Anim Batch Dup Item Object References, Selected:bool=~) {is=title K2Node_FunctionEntry}
   > n1 = Sequence!(execute) -> (then_0, then_1)
   > n2 = For Each Loop(Array=Get Items())
   > Set Selected!(execute, Selected:bool=Get Selected!() -> (Selected), self:BP Anim Batch Dup Item Object Reference=@n2) -> (then, Output_Get:bool)
   Return Node!(execute=@n1.then_1, Updated Items:Array of BP Anim Batch Dup Item Object References=Get Items())
  ]}
 Update Target Path Text Widgets {nodes [
   Update Target Path Text Widgets() {is=title K2Node_FunctionEntry}
   > n1 = Sequence!(execute) -> (then_0, then_1, then_2:Exec)
   > Set Text(
    Text="To Text (String)"(InString=Get Target LS Path())
    self:Text Object Reference=Get TargetLSPathText() -> (TargetLSPathText:Text Object Reference)
   )
   Set Text(
    execute=@n1.then_1
    Text="To Text (String)"(InString=Get Target AS Path())
    self=Get TargetASPathText() -> (TargetASPathText:Text Object Reference)
   )
   Set Text(
    execute=@n1.then_2
    Text="To Text (String)"(InString=Get Target AM Path())
    self=Get TargetAMPathText() -> (TargetAMPathText:Text Object Reference)
   )
  ]}
]
variables [
 AC {type=AnimatableCreature_C, default=null}
 Anim Export Options {type=AnimSeqExportOption, default=null}
 Bake To New Character {type=boolean}
 Current Anim Montage Paths {type="array<string>", default []}
 Current List Items {type="array<BP_AnimBatchDup_Item_C>", default []}
 Documentation URL {type=string, default=https://epicgames.atlassian.net/wiki/spaces/PP/pages/281937101/Animation+Batch+Duplicate+tool}
 IK Retargeter {type=IKRetargeter, default=/CosmeticCompanions/Assets/Quadruped/TalonScone/Meshes/RTG_TalonScone}
 IKRetargeter Border Color {type=SlateColor, default {colorUseRule=UseColor_Specified, specifiedColor=<rgba 0.0194 0.0194 0.0194 1>}}
 Include AnimMontages {type=boolean, default=true}
 Include AnimSequences {type=boolean, default=true}
 Include LevelSequences {type=boolean, default=true}
 Include Target AM Subfolder {type=boolean, default=true}
 Include Target AS Subfolder {type=boolean, default=true}
 Include Target LS Subfolder {type=boolean, default=true}
 Input Source Path {type=string, default=""}
 Input Source Path Recursive {type=boolean}
 Log {type=string, default=""}
 Retain Additive Flags {type=boolean}
 Target {type=BP_AnimBatchDup_SubItem_C, default=null}
 Target AC {type=AnimatableCreature_C, default=null}
 Target AM Path {type=string, default=""}
 Target AS Path {type=string, default=""}
 Target LS Path {type=string, default=""}
 Using Source Custom Path {type=boolean}
 Using Target Custom Paths {type=boolean}
]
widgets {
 info {rootWidgetClass=&StackBox, widgetCount=121}
 widgets [
  StackBox_0 {widgetClassPath=&StackBox, widget=StackBox_0}
  VerticalBox_0 {
   widgetClassPath=&VerticalBox
   parent=StackBox_0
   widget=VerticalBox_0
   slot {class=StackBoxSlot, properties {size {sizeRule=Fill, value=1}}}
  }
  VerticalBox_1 {widgetClassPath=&VerticalBox, parent=VerticalBox_0, widget=VerticalBox_1, slot {class=VerticalBoxSlot}}
  ACandPathsHorizontalBox {widgetClassPath=&HorizontalBox, parent=VerticalBox_1, widget=ACandPathsHorizontalBox, slot {class=VerticalBoxSlot}}
  VerticalBox_2 {
   widgetClassPath=&VerticalBox
   parent=ACandPathsHorizontalBox
   widget=VerticalBox_2
   slot {class=HorizontalBoxSlot, properties {padding=<ltrb 10 10 10 10>, size {sizeRule=Fill, value=1}}}
  }
  HorizontalBox {
   widgetClassPath=&HorizontalBox
   parent=VerticalBox_2
   widget=HorizontalBox
   slot {class=VerticalBoxSlot, properties {padding=<ltrb 0 0 0 10>}}
  }
  TextBlock {
   widgetClassPath=&TextBlock
   parent=HorizontalBox
   widget=TextBlock
   properties {
    text=Source Level Sequences from
    font {
     bForceMonospaced=false
     bMaterialIsStencil=false
     fontMaterial=null
     fontObject=&Roboto
     letterSpacing=0
     monospacedWidth=1
     size=10.5
     skewAmount=0
     typefaceFontName=Bold
     outlineSettings {
      bApplyOutlineToDropShadows=false
      bMiteredCorners=false
      bSeparateFillAlpha=false
      outlineMaterial=null
      outlineSize=0
      outlineColor=<rgba 0 0 0 1>
     }
    }
   }
   slot {class=HorizontalBoxSlot, properties {verticalAlignment=VAlign_Center, padding=<ltrb 0 0 5 0>}}
  }
  SourceAnimationsComboBox_1 {
   widgetClassPath=/Script/Blutility.EditorUtilityComboBoxString
   bIsVariable=true
   parent=HorizontalBox
   widget=SourceAnimationsComboBox_1
   properties {
    selectedOption=AC Asset
    defaultOptions [AC Asset, Path]
    font {
     bForceMonospaced=false
     bMaterialIsStencil=false
     fontMaterial=null
     fontObject=&Roboto
     letterSpacing=0
     monospacedWidth=1
     size=10.5
     skewAmount=0
     typefaceFontName=Regular
     outlineSettings {
      bApplyOutlineToDropShadows=false
      bMiteredCorners=false
      bSeparateFillAlpha=false
      outlineMaterial=null
      outlineSize=0
      outlineColor=<rgba 0 0 0 1>
     }
    }
    foregroundColor {colorUseRule=UseColor_Specified, specifiedColor=<rgba 0.5271 0.5271 0.5271 1>}
   }
   slot {class=HorizontalBoxSlot, properties {size {sizeRule=Fill, value=1}}}
  }
  SourceACPropertyViewBox {
   widgetClassPath=&HorizontalBox
   bIsVariable=true
   parent=VerticalBox_2
   widget=SourceACPropertyViewBox
   slot {class=VerticalBoxSlot}
  }
  SourceACPropertyView {
   widgetClassPath=&SinglePropertyView
   parent=SourceACPropertyViewBox
   widget=SourceACPropertyView
   properties {nameOverride=Source AC, propertyName=AC}
   slot {class=HorizontalBoxSlot, properties {verticalAlignment=VAlign_Top, size {sizeRule=Fill, value=1}}}
  }
  SourcePathPropertyBox {
   widgetClassPath=&VerticalBox
   bIsVariable=true
   parent=VerticalBox_2
   widget=SourcePathPropertyBox
   slot {class=VerticalBoxSlot, properties {verticalAlignment=VAlign_Top, padding=<ltrb 10 0 0 0>}}
  }
  HorizontalBox_123 {
   widgetClassPath=&HorizontalBox
   parent=SourcePathPropertyBox
   widget=HorizontalBox_123
   slot {class=VerticalBoxSlot, properties {padding=<ltrb 0 0 0 5>, size {sizeRule=Fill, value=1}}}
  }
  TextBlock_103 {
   widgetClassPath=&TextBlock
   parent=HorizontalBox_123
   widget=TextBlock_103
   properties {
    text=Path
    font {
     bForceMonospaced=false
     bMaterialIsStencil=false
     fontMaterial=null
     fontObject=&Roboto
     letterSpacing=0
     monospacedWidth=1
     size=9
     skewAmount=0
     typefaceFontName=Bold
     outlineSettings {
      bApplyOutlineToDropShadows=false
      bMiteredCorners=false
      bSeparateFillAlpha=false
      outlineMaterial=null
      outlineSize=0
      outlineColor=<rgba 0 0 0 1>
     }
    }
   }
   slot {class=HorizontalBoxSlot, properties {verticalAlignment=VAlign_Center, padding=<ltrb 0 0 5 0>}}
  }
  InputSourcePathEditableTextBox {
   widgetClassPath=&EditorUtilityEditableTextBox
   bIsVariable=true
   parent=HorizontalBox_123
   widget=InputSourcePathEditableTextBox
   slot {class=HorizontalBoxSlot, properties {size {sizeRule=Fill, value=1}}}
  }
  HorizontalBox_124 {widgetClassPath=&HorizontalBox, parent=SourcePathPropertyBox, widget=HorizontalBox_124, slot {class=VerticalBoxSlot}}
  TextBlock_203 {
   widgetClassPath=&TextBlock
   parent=HorizontalBox_124
   widget=TextBlock_203
   properties {
    text=Recursive
    font {
     bForceMonospaced=false
     bMaterialIsStencil=false
     fontMaterial=null
     fontObject=&Roboto
     letterSpacing=0
     monospacedWidth=1
     size=7.5
     skewAmount=0
     typefaceFontName=Regular
     outlineSettings {
      bApplyOutlineToDropShadows=false
      bMiteredCorners=false
      bSeparateFillAlpha=false
      outlineMaterial=null
      outlineSize=0
      outlineColor=<rgba 0 0 0 1>
     }
    }
   }
   slot {class=HorizontalBoxSlot, properties {verticalAlignment=VAlign_Center, padding=<ltrb 0 0 5 0>}}
  }
  InputPathRecursiveCheckbox {
   widgetClassPath=&EditorUtilityCheckBox
   bIsVariable=true
   parent=HorizontalBox_124
   widget=InputPathRecursiveCheckbox
   properties {widgetStyle {
     checkBoxType=CheckBox
     backgroundHoveredImage {
      bIsDynamicallyLoaded=false
      drawAs=RoundedBox
      imageType=NoImage
      mirroring=NoMirror
      resourceName=None
      resourceObject=null
      tiling=NoTile
      imageSize=<18 18>
      margin=<ltrb 0 0 0 0>
      outlineSettings {
       bUseBrushTransparency=false
       roundingType=FixedRadius
       width=1
       color {colorUseRule=UseColor_ColorTable, specifiedColor=<rgba 0 0 0 0>}
       cornerRadii=<3 3 3 3>
      }
      tintColor {colorUseRule=UseColor_Specified, specifiedColor=<rgba 0.0048 0.0048 0.0048 1>}
      uVRegion {bIsValid=false, max=<0 0>, min=<0 0>}
     }
     backgroundImage {
      bIsDynamicallyLoaded=false
      drawAs=RoundedBox
      imageType=NoImage
      mirroring=NoMirror
      resourceName=None
      resourceObject=null
      tiling=NoTile
      imageSize=<18 18>
      margin=<ltrb 0 0 0 0>
      outlineSettings {
       bUseBrushTransparency=false
       roundingType=FixedRadius
       width=1
       color {colorUseRule=UseColor_ColorTable, specifiedColor=<rgba 0 0 0 0>}
       cornerRadii=<3 3 3 3>
      }
      tintColor {colorUseRule=UseColor_Specified, specifiedColor=<rgba 0.0048 0.0048 0.0048 1>}
      uVRegion {bIsValid=false, max=<0 0>, min=<0 0>}
     }
     backgroundPressedImage {
      bIsDynamicallyLoaded=false
      drawAs=RoundedBox
      imageType=NoImage
      mirroring=NoMirror
      resourceName=None
      resourceObject=null
      tiling=NoTile
      imageSize=<18 18>
      margin=<ltrb 0 0 0 0>
      outlineSettings {
       bUseBrushTransparency=false
       roundingType=FixedRadius
       width=1
       color {colorUseRule=UseColor_ColorTable, specifiedColor=<rgba 0 0 0 0>}
       cornerRadii=<3 3 3 3>
      }
      tintColor {colorUseRule=UseColor_Specified, specifiedColor=<rgba 0.0048 0.0048 0.0048 1>}
      uVRegion {bIsValid=false, max=<0 0>, min=<0 0>}
     }
     borderBackgroundColor {colorUseRule=UseColor_Specified, specifiedColor=<rgba 1 1 1 1>}
     checkedForeground {colorUseRule=UseColor_Specified, specifiedColor=<rgba 1 1 1 1>}
     checkedHoveredForeground {colorUseRule=UseColor_Specified, specifiedColor=<rgba 1 1 1 1>}
     checkedHoveredImage {
      bIsDynamicallyLoaded=false
      drawAs=Image
      imageType=Vector
      mirroring=NoMirror
      resourceName=../../../Engine/Content/Slate/Starship/CoreWidgets/CheckBox/check.svg
      resourceObject=null
      tiling=NoTile
      imageSize=<16 16>
      margin=<ltrb 0 0 0 0>
      outlineSettings {
       bUseBrushTransparency=false
       roundingType=HalfHeightRadius
       width=0
       color {colorUseRule=UseColor_Specified, specifiedColor=<rgba 0 0 0 0>}
       cornerRadii=<0 0 0 0>
      }
      tintColor {colorUseRule=UseColor_Specified, specifiedColor=<rgba 0.0044 0.2384 1 1>}
      uVRegion {bIsValid=false, max=<0 0>, min=<0 0>}
     }
     checkedImage {
      bIsDynamicallyLoaded=false
      drawAs=Image
      imageType=Vector
      mirroring=NoMirror
      resourceName=../../../Engine/Content/Slate/Starship/CoreWidgets/CheckBox/check.svg
      resourceObject=null
      tiling=NoTile
      imageSize=<16 16>
      margin=<ltrb 0 0 0 0>
      outlineSettings {
       bUseBrushTransparency=false
       roundingType=HalfHeightRadius
       width=0
       color {colorUseRule=UseColor_Specified, specifiedColor=<rgba 0 0 0 0>}
       cornerRadii=<0 0 0 0>
      }
      tintColor {colorUseRule=UseColor_Specified, specifiedColor=<rgba 0 0.162 0.7454 1>}
      uVRegion {bIsValid=false, max=<0 0>, min=<0 0>}
     }
     checkedPressedForeground {colorUseRule=UseColor_Specified, specifiedColor=<rgba 1 1 1 1>}
     checkedPressedImage {
      bIsDynamicallyLoaded=false
      drawAs=Image
      imageType=Vector
      mirroring=NoMirror
      resourceName=../../../Engine/Content/Slate/Starship/CoreWidgets/CheckBox/check.svg
      resourceObject=null
      tiling=NoTile
      imageSize=<16 16>
      margin=<ltrb 0 0 0 0>
      outlineSettings {
       bUseBrushTransparency=false
       roundingType=HalfHeightRadius
       width=0
       color {colorUseRule=UseColor_Specified, specifiedColor=<rgba 0 0 0 0>}
       cornerRadii=<0 0 0 0>
      }
      tintColor {colorUseRule=UseColor_Specified, specifiedColor=<rgba 0 0.0802 0.3515 1>}
      uVRegion {bIsValid=false, max=<0 0>, min=<0 0>}
     }
     checkedSlateSound {resourceObject=null}
     foregroundColor {colorUseRule=UseColor_Specified, specifiedColor=<rgba 1 1 1 1>}
     hoveredForeground {colorUseRule=UseColor_Specified, specifiedColor=<rgba 1 1 1 1>}
     hoveredSlateSound {resourceObject=null}
     padding=<ltrb 0 0 0 0>
     pressedForeground {colorUseRule=UseColor_Specified, specifiedColor=<rgba 1 1 1 1>}
     uncheckedHoveredImage {
      bIsDynamicallyLoaded=false
      drawAs=NoDrawType
      imageType=NoImage
      mirroring=NoMirror
      resourceName=None
      resourceObject=null
      tiling=NoTile
      imageSize=<0 0>
      margin=<ltrb 0 0 0 0>
      outlineSettings {
       bUseBrushTransparency=false
       roundingType=HalfHeightRadius
       width=0
       color {colorUseRule=UseColor_Specified, specifiedColor=<rgba 0 0 0 0>}
       cornerRadii=<0 0 0 0>
      }
      tintColor {colorUseRule=UseColor_Specified, specifiedColor=<rgba 1 1 1 1>}
      uVRegion {bIsValid=false, max=<0 0>, min=<0 0>}
     }
     uncheckedImage {
      bIsDynamicallyLoaded=false
      drawAs=NoDrawType
      imageType=NoImage
      mirroring=NoMirror
      resourceName=None
      resourceObject=null
      tiling=NoTile
      imageSize=<0 0>
      margin=<ltrb 0 0 0 0>
      outlineSettings {
       bUseBrushTransparency=false
       roundingType=HalfHeightRadius
       width=0
       color {colorUseRule=UseColor_Specified, specifiedColor=<rgba 0 0 0 0>}
       cornerRadii=<0 0 0 0>
      }
      tintColor {colorUseRule=UseColor_Specified, specifiedColor=<rgba 1 1 1 1>}
      uVRegion {bIsValid=false, max=<0 0>, min=<0 0>}
     }
     uncheckedPressedImage {
      bIsDynamicallyLoaded=false
      drawAs=NoDrawType
      imageType=NoImage
      mirroring=NoMirror
      resourceName=None
      resourceObject=null
      tiling=NoTile
      imageSize=<0 0>
      margin=<ltrb 0 0 0 0>
      outlineSettings {
       bUseBrushTransparency=false
       roundingType=HalfHeightRadius
       width=0
       color {colorUseRule=UseColor_Specified, specifiedColor=<rgba 0 0 0 0>}
       cornerRadii=<0 0 0 0>
      }
      tintColor {colorUseRule=UseColor_Specified, specifiedColor=<rgba 1 1 1 1>}
      uVRegion {bIsValid=false, max=<0 0>, min=<0 0>}
     }
     uncheckedSlateSound {resourceObject=null}
     undeterminedForeground {colorUseRule=UseColor_Specified, specifiedColor=<rgba 1 1 1 1>}
     undeterminedHoveredImage {
      bIsDynamicallyLoaded=false
      drawAs=Image
      imageType=Vector
      mirroring=NoMirror
      resourceName=../../../Engine/Content/Slate/Starship/CoreWidgets/CheckBox/indeterminate.svg
      resourceObject=null
      tiling=NoTile
      imageSize=<16 16>
      margin=<ltrb 0 0 0 0>
      outlineSettings {
       bUseBrushTransparency=false
       roundingType=HalfHeightRadius
       width=0
       color {colorUseRule=UseColor_Specified, specifiedColor=<rgba 0 0 0 0>}
       cornerRadii=<0 0 0 0>
      }
      tintColor {colorUseRule=UseColor_Specified, specifiedColor=<rgba 0.0044 0.2384 1 1>}
      uVRegion {bIsValid=false, max=<0 0>, min=<0 0>}
     }
     undeterminedImage {
      bIsDynamicallyLoaded=false
      drawAs=Image
      imageType=Vector
      mirroring=NoMirror
      resourceName=../../../Engine/Content/Slate/Starship/CoreWidgets/CheckBox/indeterminate.svg
      resourceObject=null
      tiling=NoTile
      imageSize=<16 16>
      margin=<ltrb 0 0 0 0>
      outlineSettings {
       bUseBrushTransparency=false
       roundingType=HalfHeightRadius
       width=0
       color {colorUseRule=UseColor_Specified, specifiedColor=<rgba 0 0 0 0>}
       cornerRadii=<0 0 0 0>
      }
      tintColor {colorUseRule=UseColor_Specified, specifiedColor=<rgba 0 0.162 0.7454 1>}
      uVRegion {bIsValid=false, max=<0 0>, min=<0 0>}
     }
     undeterminedPressedImage {
      bIsDynamicallyLoaded=false
      drawAs=Image
      imageType=Vector
      mirroring=NoMirror
      resourceName=../../../Engine/Content/Slate/Starship/CoreWidgets/CheckBox/indeterminate.svg
      resourceObject=null
      tiling=NoTile
      imageSize=<16 16>
      margin=<ltrb 0 0 0 0>
      outlineSettings {
       bUseBrushTransparency=false
       roundingType=HalfHeightRadius
       width=0
       color {colorUseRule=UseColor_Specified, specifiedColor=<rgba 0 0 0 0>}
       cornerRadii=<0 0 0 0>
      }
      tintColor {colorUseRule=UseColor_Specified, specifiedColor=<rgba 0 0.0802 0.3515 1>}
      uVRegion {bIsValid=false, max=<0 0>, min=<0 0>}
     }
    }}
   slot {class=HorizontalBoxSlot, properties {padding=<ltrb 0 0 5 0>}}
  }
  TextBlock_833 {
   widgetClassPath=&TextBlock
   parent=HorizontalBox_124
   widget=TextBlock_833
   properties {
    autoWrapText=true
    text="It may take a long time, use with caution."
    colorAndOpacity {colorUseRule=UseColor_Specified, specifiedColor=<rgba 1 0 0 1>}
    font {
     bForceMonospaced=false
     bMaterialIsStencil=false
     fontMaterial=null
     fontObject=&Roboto
     letterSpacing=0
     monospacedWidth=1
     size=7.5
     skewAmount=0
     typefaceFontName=Regular
     outlineSettings {
      bApplyOutlineToDropShadows=false
      bMiteredCorners=false
      bSeparateFillAlpha=false
      outlineMaterial=null
      outlineSize=0
      outlineColor=<rgba 0 0 0 1>
     }
    }
   }
   slot {class=HorizontalBoxSlot, properties {verticalAlignment=VAlign_Center, size {sizeRule=Fill, value=1}}}
  }
  VerticalBox_3 {
   widgetClassPath=&VerticalBox
   parent=ACandPathsHorizontalBox
   widget=VerticalBox_3
   slot {class=HorizontalBoxSlot, properties {verticalAlignment=VAlign_Top, padding=<ltrb 10 10 10 10>, size {sizeRule=Fill, value=1}}}
  }
  HorizontalBox_0 {
   widgetClassPath=&HorizontalBox
   parent=VerticalBox_3
   widget=HorizontalBox_0
   slot {class=VerticalBoxSlot, properties {padding=<ltrb 0 0 0 10>}}
  }
  TextBlock_121 {
   widgetClassPath=&TextBlock
   parent=HorizontalBox_0
   widget=TextBlock_121
   properties {
    text=Target Animations to
    font {
     bForceMonospaced=false
     bMaterialIsStencil=false
     fontMaterial=null
     fontObject=&Roboto
     letterSpacing=0
     monospacedWidth=1
     size=10.5
     skewAmount=0
     typefaceFontName=Bold
     outlineSettings {
      bApplyOutlineToDropShadows=false
      bMiteredCorners=false
      bSeparateFillAlpha=false
      outlineMaterial=null
      outlineSize=0
      outlineColor=<rgba 0 0 0 1>
     }
    }
   }
   slot {class=HorizontalBoxSlot, properties {verticalAlignment=VAlign_Center, padding=<ltrb 0 0 5 0>}}
  }
  TargetAnimationsComboBox {
   widgetClassPath=&EditorUtilityComboBoxString
   bIsVariable=true
   parent=HorizontalBox_0
   widget=TargetAnimationsComboBox
   properties {
    selectedOption=AC Asset
    defaultOptions [AC Asset, Paths]
    font {
     bForceMonospaced=false
     bMaterialIsStencil=false
     fontMaterial=null
     fontObject=&Roboto
     letterSpacing=0
     monospacedWidth=1
     size=10.5
     skewAmount=0
     typefaceFontName=Regular
     outlineSettings {
      bApplyOutlineToDropShadows=false
      bMiteredCorners=false
      bSeparateFillAlpha=false
      outlineMaterial=null
      outlineSize=0
      outlineColor=<rgba 0 0 0 1>
     }
    }
    foregroundColor {colorUseRule=UseColor_Specified, specifiedColor=<rgba 0.5271 0.5271 0.5271 1>}
   }
   slot {class=HorizontalBoxSlot, properties {size {sizeRule=Fill, value=1}}}
  }
  DocumentationButton {
   widgetClassPath=&EditorUtilityButton
   parent=HorizontalBox_0
   widget=DocumentationButton
   properties {widgetStyle {
     clickedSlateSound {resourceObject=null}
     disabled {
      bIsDynamicallyLoaded=false
      drawAs=RoundedBox
      imageType=NoImage
      mirroring=NoMirror
      resourceName=None
      resourceObject=null
      tiling=NoTile
      imageSize=<26 26>
      margin=<ltrb 0 0 0 0>
      outlineSettings {
       bUseBrushTransparency=false
       roundingType=FixedRadius
       width=1
       color {colorUseRule=UseColor_ColorTable, specifiedColor=<rgba 0 0 0 0>}
       cornerRadii=<4 4 4 4>
      }
      tintColor {colorUseRule=UseColor_Specified, specifiedColor=<rgba 0.0395 0.0395 0.0395 1>}
      uVRegion {bIsValid=false, max=<0 0>, min=<0 0>}
     }
     disabledForeground {colorUseRule=UseColor_ColorTable, specifiedColor=<rgba 0 0 0 0>}
     hovered {
      bIsDynamicallyLoaded=false
      drawAs=RoundedBox
      imageType=NoImage
      mirroring=NoMirror
      resourceName=None
      resourceObject=null
      tiling=NoTile
      imageSize=<26 26>
      margin=<ltrb 0 0 0 0>
      outlineSettings {
       bUseBrushTransparency=false
       roundingType=FixedRadius
       width=1
       color {colorUseRule=UseColor_ColorTable, specifiedColor=<rgba 0 0 0 0>}
       cornerRadii=<4 4 4 4>
      }
      tintColor {colorUseRule=UseColor_Specified, specifiedColor=<rgba 0.0953 0.0953 0.0953 1>}
      uVRegion {bIsValid=false, max=<0 0>, min=<0 0>}
     }
     hoveredForeground {colorUseRule=UseColor_ColorTable, specifiedColor=<rgba 0 0 0 0>}
     hoveredSlateSound {resourceObject=null}
     normal {
      bIsDynamicallyLoaded=false
      drawAs=RoundedBox
      imageType=NoImage
      mirroring=NoMirror
      resourceName=None
      resourceObject=null
      tiling=NoTile
      imageSize=<26 26>
      margin=<ltrb 0 0 0 0>
      outlineSettings {
       bUseBrushTransparency=false
       roundingType=FixedRadius
       width=1
       color {colorUseRule=UseColor_ColorTable, specifiedColor=<rgba 0 0 0 0>}
       cornerRadii=<4 4 4 4>
      }
      tintColor {colorUseRule=UseColor_Specified, specifiedColor=<rgba 0.0395 0.0395 0.0395 1>}
      uVRegion {bIsValid=false, max=<0 0>, min=<0 0>}
     }
     normalForeground {colorUseRule=UseColor_ColorTable, specifiedColor=<rgba 0 0 0 0>}
     normalPadding=<ltrb 5 5 5 5>
     pressed {
      bIsDynamicallyLoaded=false
      drawAs=RoundedBox
      imageType=NoImage
      mirroring=NoMirror
      resourceName=None
      resourceObject=null
      tiling=NoTile
      imageSize=<26 26>
      margin=<ltrb 0 0 0 0>
      outlineSettings {
       bUseBrushTransparency=false
       roundingType=FixedRadius
       width=1
       color {colorUseRule=UseColor_ColorTable, specifiedColor=<rgba 0 0 0 0>}
       cornerRadii=<4 4 4 4>
      }
      tintColor {colorUseRule=UseColor_Specified, specifiedColor=<rgba 0.0284 0.0284 0.0284 1>}
      uVRegion {bIsValid=false, max=<0 0>, min=<0 0>}
     }
     pressedForeground {colorUseRule=UseColor_ColorTable, specifiedColor=<rgba 0 0 0 0>}
     pressedPadding=<ltrb 5 5 5 5>
     pressedSlateSound {resourceObject=null}
    }}
   slot {class=HorizontalBoxSlot, properties {padding=<ltrb 10 0 0 0>}}
  }
  Image_83 {
   widgetClassPath=&Image
   bIsVariable=true
   parent=DocumentationButton
   widget=Image_83
   properties {brush {
     bIsDynamicallyLoaded=false
     drawAs=Image
     imageType=FullColor
     mirroring=NoMirror
     resourceName=None
     resourceObject=/AnimInEdTools/AnimBatchDuplicate/Icons/AnimBatchDup_Icon_Help
     tiling=NoTile
     imageSize=<16 16>
     margin=<ltrb 0 0 0 0>
     outlineSettings {
      bUseBrushTransparency=false
      roundingType=HalfHeightRadius
      width=0
      color {colorUseRule=UseColor_Specified, specifiedColor=<rgba 0 0 0 0>}
      cornerRadii=<0 0 0 0>
     }
     tintColor {colorUseRule=UseColor_Specified, specifiedColor=<rgba 1 1 1 1>}
     uVRegion {bIsValid=false, max=<0 0>, min=<0 0>}
    }}
   slot {class=ButtonSlot, properties {horizontalAlignment=HAlign_Fill, verticalAlignment=VAlign_Fill, padding=<ltrb 0 0 0 0>}}
  }
  TargetACPropertyViewBox {
   widgetClassPath=&HorizontalBox
   bIsVariable=true
   parent=VerticalBox_3
   widget=TargetACPropertyViewBox
   slot {class=VerticalBoxSlot, properties {verticalAlignment=VAlign_Top}}
  }
  VerticalBox_117 {
   widgetClassPath=&VerticalBox
   parent=TargetACPropertyViewBox
   widget=VerticalBox_117
   slot {class=HorizontalBoxSlot, properties {size {sizeRule=Fill, value=1}}}
  }
  TargetACPropertyView {
   widgetClassPath=&SinglePropertyView
   parent=VerticalBox_117
   widget=TargetACPropertyView
   properties {propertyName=Target AC}
   slot {class=VerticalBoxSlot, properties {padding=<ltrb 0 0 0 10>}}
  }
  TargetPathsfromACAsset {
   widgetClassPath=&VerticalBox
   bIsVariable=true
   parent=VerticalBox_117
   widget=TargetPathsfromACAsset
   slot {class=VerticalBoxSlot, properties {padding=<ltrb 0 0 0 5>}}
  }
  ACTargetLevelSequenceBox {
   widgetClassPath=&HorizontalBox
   bIsVariable=true
   parent=TargetPathsfromACAsset
   widget=ACTargetLevelSequenceBox
   slot {class=VerticalBoxSlot, properties {padding=<ltrb 0 0 0 5>}}
  }
  SubfolderTargetLSPathACButton {
   widgetClassPath=&EditorUtilityButton
   parent=ACTargetLevelSequenceBox
   widget=SubfolderTargetLSPathACButton
   properties {widgetStyle {
     clickedSlateSound {resourceObject=null}
     disabled {
      bIsDynamicallyLoaded=false
      drawAs=RoundedBox
      imageType=NoImage
      mirroring=NoMirror
      resourceName=None
      resourceObject=null
      tiling=NoTile
      imageSize=<0 0>
      margin=<ltrb 0 0 0 0>
      outlineSettings {
       bUseBrushTransparency=false
       roundingType=FixedRadius
       width=1
       color {colorUseRule=UseColor_ColorTable, specifiedColor=<rgba 0 0 0 0>}
       cornerRadii=<4 4 4 4>
      }
      tintColor {colorUseRule=UseColor_Specified, specifiedColor=<rgba 0.0395 0.0395 0.0395 1>}
      uVRegion {bIsValid=false, max=<0 0>, min=<0 0>}
     }
     disabledForeground {colorUseRule=UseColor_ColorTable, specifiedColor=<rgba 0 0 0 0>}
     hovered {
      bIsDynamicallyLoaded=false
      drawAs=RoundedBox
      imageType=NoImage
      mirroring=NoMirror
      resourceName=None
      resourceObject=null
      tiling=NoTile
      imageSize=<18 18>
      margin=<ltrb 0 0 0 0>
      outlineSettings {
       bUseBrushTransparency=false
       roundingType=FixedRadius
       width=1
       color {colorUseRule=UseColor_ColorTable, specifiedColor=<rgba 0 0 0 0>}
       cornerRadii=<4 4 4 4>
      }
      tintColor {colorUseRule=UseColor_Specified, specifiedColor=<rgba 0.599 0.3697 0.1487 1>}
      uVRegion {bIsValid=false, max=<0 0>, min=<0 0>}
     }
     hoveredForeground {colorUseRule=UseColor_ColorTable, specifiedColor=<rgba 0 0 0 0>}
     hoveredSlateSound {resourceObject=null}
     normal {
      bIsDynamicallyLoaded=false
      drawAs=RoundedBox
      imageType=NoImage
      mirroring=NoMirror
      resourceName=None
      resourceObject=null
      tiling=NoTile
      imageSize=<18 18>
      margin=<ltrb 0 0 0 0>
      outlineSettings {
       bUseBrushTransparency=false
       roundingType=FixedRadius
       width=1
       color {colorUseRule=UseColor_ColorTable, specifiedColor=<rgba 0 0 0 0>}
       cornerRadii=<4 4 4 4>
      }
      tintColor {colorUseRule=UseColor_Specified, specifiedColor=<rgba 0.3968 0.2346 0.0782 1>}
      uVRegion {bIsValid=false, max=<0 0>, min=<0 0>}
     }
     normalForeground {colorUseRule=UseColor_ColorTable, specifiedColor=<rgba 0 0 0 0>}
     normalPadding=<ltrb 0 0 0 0>
     pressed {
      bIsDynamicallyLoaded=false
      drawAs=RoundedBox
      imageType=NoImage
      mirroring=NoMirror
      resourceName=None
      resourceObject=null
      tiling=NoTile
      imageSize=<18 18>
      margin=<ltrb 0 0 0 0>
      outlineSettings {
       bUseBrushTransparency=false
       roundingType=FixedRadius
       width=1
       color {colorUseRule=UseColor_ColorTable, specifiedColor=<rgba 0 0 0 0>}
       cornerRadii=<4 4 4 4>
      }
      tintColor {colorUseRule=UseColor_Specified, specifiedColor=<rgba 0.283 0.1607 0.0427 1>}
      uVRegion {bIsValid=false, max=<0 0>, min=<0 0>}
     }
     pressedForeground {colorUseRule=UseColor_ColorTable, specifiedColor=<rgba 0 0 0 0>}
     pressedPadding=<ltrb 0 0 0 0>
     pressedSlateSound {resourceObject=null}
    }}
   slot {class=HorizontalBoxSlot, properties {padding=<ltrb 0 0 5 0>}}
  }
  Image_158 {
   widgetClassPath=&Image
   bIsVariable=true
   parent=SubfolderTargetLSPathACButton
   widget=Image_158
   properties {brush {
     bIsDynamicallyLoaded=false
     drawAs=Image
     imageType=FullColor
     mirroring=NoMirror
     resourceName=None
     resourceObject=/AnimInEdTools/AnimBatchDuplicate/Icons/AnimBatchDup_Icon_Folder
     tiling=NoTile
     imageSize=<14 14>
     margin=<ltrb 0 0 0 0>
     outlineSettings {
      bUseBrushTransparency=false
      roundingType=HalfHeightRadius
      width=0
      color {colorUseRule=UseColor_Specified, specifiedColor=<rgba 0 0 0 0>}
      cornerRadii=<0 0 0 0>
     }
     tintColor {colorUseRule=UseColor_Specified, specifiedColor=<rgba 0.1719 0.0979 0.0266 1>}
     uVRegion {bIsValid=false, max=<0 0>, min=<0 0>}
    }}
   slot {class=ButtonSlot, properties {padding=<ltrb 3 3 3 3>}}
  }
  TargetLSPathText {
   widgetClassPath=&TextBlock
   bIsVariable=true
   parent=ACTargetLevelSequenceBox
   widget=TargetLSPathText
   properties {
    text=Text Block
    colorAndOpacity {colorUseRule=UseColor_Specified, specifiedColor=<rgba 0.5776 0.0802 0.0802 1>}
    font {
     bForceMonospaced=false
     bMaterialIsStencil=false
     fontMaterial=null
     fontObject=&Roboto
     letterSpacing=0
     monospacedWidth=1
     size=9
     skewAmount=0
     typefaceFontName=Bold
     outlineSettings {
      bApplyOutlineToDropShadows=false
      bMiteredCorners=false
      bSeparateFillAlpha=false
      outlineMaterial=null
      outlineSize=0
      outlineColor=<rgba 0 0 0 1>
     }
    }
   }
   slot {class=HorizontalBoxSlot, properties {verticalAlignment=VAlign_Center}}
  }
  TargetLSSubfolderText {
   widgetClassPath=&TextBlock
   bIsVariable=true
   parent=ACTargetLevelSequenceBox
   widget=TargetLSSubfolderText
   properties {
    text=/Subfolder
    colorAndOpacity {colorUseRule=UseColor_Specified, specifiedColor=<rgba 0.3968 0.2346 0.0782 1>}
    font {
     bForceMonospaced=false
     bMaterialIsStencil=false
     fontMaterial=null
     fontObject=&Roboto
     letterSpacing=0
     monospacedWidth=1
     size=9
     skewAmount=0
     typefaceFontName=Bold
     outlineSettings {
      bApplyOutlineToDropShadows=false
      bMiteredCorners=false
      bSeparateFillAlpha=false
      outlineMaterial=null
      outlineSize=0
      outlineColor=<rgba 0 0 0 1>
     }
    }
   }
   slot {class=HorizontalBoxSlot, properties {verticalAlignment=VAlign_Center}}
  }
  ACTargetAnimSequenceBox {
   widgetClassPath=&HorizontalBox
   bIsVariable=true
   parent=TargetPathsfromACAsset
   widget=ACTargetAnimSequenceBox
   slot {class=VerticalBoxSlot, properties {padding=<ltrb 0 0 0 5>}}
  }
  SubfolderTargetASPathACButton {
   widgetClassPath=&EditorUtilityButton
   parent=ACTargetAnimSequenceBox
   widget=SubfolderTargetASPathACButton
   properties {widgetStyle {
     clickedSlateSound {resourceObject=null}
     disabled {
      bIsDynamicallyLoaded=false
      drawAs=RoundedBox
      imageType=NoImage
      mirroring=NoMirror
      resourceName=None
      resourceObject=null
      tiling=NoTile
      imageSize=<0 0>
      margin=<ltrb 0 0 0 0>
      outlineSettings {
       bUseBrushTransparency=false
       roundingType=FixedRadius
       width=1
       color {colorUseRule=UseColor_ColorTable, specifiedColor=<rgba 0 0 0 0>}
       cornerRadii=<4 4 4 4>
      }
      tintColor {colorUseRule=UseColor_Specified, specifiedColor=<rgba 0.0395 0.0395 0.0395 1>}
      uVRegion {bIsValid=false, max=<0 0>, min=<0 0>}
     }
     disabledForeground {colorUseRule=UseColor_ColorTable, specifiedColor=<rgba 0 0 0 0>}
     hovered {
      bIsDynamicallyLoaded=false
      drawAs=RoundedBox
      imageType=NoImage
      mirroring=NoMirror
      resourceName=None
      resourceObject=null
      tiling=NoTile
      imageSize=<18 18>
      margin=<ltrb 0 0 0 0>
      outlineSettings {
       bUseBrushTransparency=false
       roundingType=FixedRadius
       width=1
       color {colorUseRule=UseColor_ColorTable, specifiedColor=<rgba 0 0 0 0>}
       cornerRadii=<4 4 4 4>
      }
      tintColor {colorUseRule=UseColor_Specified, specifiedColor=<rgba 0.599 0.3697 0.1487 1>}
      uVRegion {bIsValid=false, max=<0 0>, min=<0 0>}
     }
     hoveredForeground {colorUseRule=UseColor_ColorTable, specifiedColor=<rgba 0 0 0 0>}
     hoveredSlateSound {resourceObject=null}
     normal {
      bIsDynamicallyLoaded=false
      drawAs=RoundedBox
      imageType=NoImage
      mirroring=NoMirror
      resourceName=None
      resourceObject=null
      tiling=NoTile
      imageSize=<18 18>
      margin=<ltrb 0 0 0 0>
      outlineSettings {
       bUseBrushTransparency=false
       roundingType=FixedRadius
       width=1
       color {colorUseRule=UseColor_ColorTable, specifiedColor=<rgba 0 0 0 0>}
       cornerRadii=<4 4 4 4>
      }
      tintColor {colorUseRule=UseColor_Specified, specifiedColor=<rgba 0.3968 0.2346 0.0782 1>}
      uVRegion {bIsValid=false, max=<0 0>, min=<0 0>}
     }
     normalForeground {colorUseRule=UseColor_ColorTable, specifiedColor=<rgba 0 0 0 0>}
     normalPadding=<ltrb 0 0 0 0>
     pressed {
      bIsDynamicallyLoaded=false
      drawAs=RoundedBox
      imageType=NoImage
      mirroring=NoMirror
      resourceName=None
      resourceObject=null
      tiling=NoTile
      imageSize=<18 18>
      margin=<ltrb 0 0 0 0>
      outlineSettings {
       bUseBrushTransparency=false
       roundingType=FixedRadius
       width=1
       color {colorUseRule=UseColor_ColorTable, specifiedColor=<rgba 0 0 0 0>}
       cornerRadii=<4 4 4 4>
      }
      tintColor {colorUseRule=UseColor_Specified, specifiedColor=<rgba 0.283 0.1607 0.0427 1>}
      uVRegion {bIsValid=false, max=<0 0>, min=<0 0>}
     }
     pressedForeground {colorUseRule=UseColor_ColorTable, specifiedColor=<rgba 0 0 0 0>}
     pressedPadding=<ltrb 0 0 0 0>
     pressedSlateSound {resourceObject=null}
    }}
   slot {class=HorizontalBoxSlot, properties {padding=<ltrb 0 0 5 0>}}
  }
  Image {
   widgetClassPath=&Image
   bIsVariable=true
   parent=SubfolderTargetASPathACButton
   widget=Image
   properties {brush {
     bIsDynamicallyLoaded=false
     drawAs=Image
     imageType=FullColor
     mirroring=NoMirror
     resourceName=None
     resourceObject=&AnimBatchDup_Icon_Folder
     tiling=NoTile
     imageSize=<14 14>
     margin=<ltrb 0 0 0 0>
     outlineSettings {
      bUseBrushTransparency=false
      roundingType=HalfHeightRadius
      width=0
      color {colorUseRule=UseColor_Specified, specifiedColor=<rgba 0 0 0 0>}
      cornerRadii=<0 0 0 0>
     }
     tintColor {colorUseRule=UseColor_Specified, specifiedColor=<rgba 0.1719 0.0979 0.0266 1>}
     uVRegion {bIsValid=false, max=<0 0>, min=<0 0>}
    }}
   slot {class=ButtonSlot, properties {padding=<ltrb 3 3 3 3>}}
  }
  TargetASPathText {
   widgetClassPath=&TextBlock
   bIsVariable=true
   parent=ACTargetAnimSequenceBox
   widget=TargetASPathText
   properties {
    text=Text Block
    colorAndOpacity {colorUseRule=UseColor_Specified, specifiedColor=<rgba 0.0802 0.1981 0.0648 1>}
    font {
     bForceMonospaced=false
     bMaterialIsStencil=false
     fontMaterial=null
     fontObject=&Roboto
     letterSpacing=0
     monospacedWidth=1
     size=9
     skewAmount=0
     typefaceFontName=Bold
     outlineSettings {
      bApplyOutlineToDropShadows=false
      bMiteredCorners=false
      bSeparateFillAlpha=false
      outlineMaterial=null
      outlineSize=0
      outlineColor=<rgba 0 0 0 1>
     }
    }
   }
   slot {class=HorizontalBoxSlot, properties {verticalAlignment=VAlign_Center}}
  }
  TargetASSubfolderText {
   widgetClassPath=&TextBlock
   bIsVariable=true
   parent=ACTargetAnimSequenceBox
   widget=TargetASSubfolderText
   properties {
    text=/Subfolder
    colorAndOpacity {colorUseRule=UseColor_Specified, specifiedColor=<rgba 0.3968 0.2346 0.0782 1>}
    font {
     bForceMonospaced=false
     bMaterialIsStencil=false
     fontMaterial=null
     fontObject=&Roboto
     letterSpacing=0
     monospacedWidth=1
     size=9
     skewAmount=0
     typefaceFontName=Bold
     outlineSettings {
      bApplyOutlineToDropShadows=false
      bMiteredCorners=false
      bSeparateFillAlpha=false
      outlineMaterial=null
      outlineSize=0
      outlineColor=<rgba 0 0 0 1>
     }
    }
   }
   slot {class=HorizontalBoxSlot, properties {verticalAlignment=VAlign_Center}}
  }
  ACTargetAnimMontagesBox {
   widgetClassPath=&HorizontalBox
   bIsVariable=true
   parent=TargetPathsfromACAsset
   widget=ACTargetAnimMontagesBox
   slot {class=VerticalBoxSlot, properties {padding=<ltrb 0 0 0 5>}}
  }
  SubfolderTargetAMPathACButton {
   widgetClassPath=&EditorUtilityButton
   parent=ACTargetAnimMontagesBox
   widget=SubfolderTargetAMPathACButton
   properties {widgetStyle {
     clickedSlateSound {resourceObject=null}
     disabled {
      bIsDynamicallyLoaded=false
      drawAs=RoundedBox
      imageType=NoImage
      mirroring=NoMirror
      resourceName=None
      resourceObject=null
      tiling=NoTile
      imageSize=<0 0>
      margin=<ltrb 0 0 0 0>
      outlineSettings {
       bUseBrushTransparency=false
       roundingType=FixedRadius
       width=1
       color {colorUseRule=UseColor_ColorTable, specifiedColor=<rgba 0 0 0 0>}
       cornerRadii=<4 4 4 4>
      }
      tintColor {colorUseRule=UseColor_Specified, specifiedColor=<rgba 0.0395 0.0395 0.0395 1>}
      uVRegion {bIsValid=false, max=<0 0>, min=<0 0>}
     }
     disabledForeground {colorUseRule=UseColor_ColorTable, specifiedColor=<rgba 0 0 0 0>}
     hovered {
      bIsDynamicallyLoaded=false
      drawAs=RoundedBox
      imageType=NoImage
      mirroring=NoMirror
      resourceName=None
      resourceObject=null
      tiling=NoTile
      imageSize=<18 18>
      margin=<ltrb 0 0 0 0>
      outlineSettings {
       bUseBrushTransparency=false
       roundingType=FixedRadius
       width=1
       color {colorUseRule=UseColor_ColorTable, specifiedColor=<rgba 0 0 0 0>}
       cornerRadii=<4 4 4 4>
      }
      tintColor {colorUseRule=UseColor_Specified, specifiedColor=<rgba 0.599 0.3697 0.1487 1>}
      uVRegion {bIsValid=false, max=<0 0>, min=<0 0>}
     }
     hoveredForeground {colorUseRule=UseColor_ColorTable, specifiedColor=<rgba 0 0 0 0>}
     hoveredSlateSound {resourceObject=null}
     normal {
      bIsDynamicallyLoaded=false
      drawAs=RoundedBox
      imageType=NoImage
      mirroring=NoMirror
      resourceName=None
      resourceObject=null
      tiling=NoTile
      imageSize=<18 18>
      margin=<ltrb 0 0 0 0>
      outlineSettings {
       bUseBrushTransparency=false
       roundingType=FixedRadius
       width=1
       color {colorUseRule=UseColor_ColorTable, specifiedColor=<rgba 0 0 0 0>}
       cornerRadii=<4 4 4 4>
      }
      tintColor {colorUseRule=UseColor_Specified, specifiedColor=<rgba 0.3968 0.2346 0.0782 1>}
      uVRegion {bIsValid=false, max=<0 0>, min=<0 0>}
     }
     normalForeground {colorUseRule=UseColor_ColorTable, specifiedColor=<rgba 0 0 0 0>}
     normalPadding=<ltrb 0 0 0 0>
     pressed {
      bIsDynamicallyLoaded=false
      drawAs=RoundedBox
      imageType=NoImage
      mirroring=NoMirror
      resourceName=None
      resourceObject=null
      tiling=NoTile
      imageSize=<18 18>
      margin=<ltrb 0 0 0 0>
      outlineSettings {
       bUseBrushTransparency=false
       roundingType=FixedRadius
       width=1
       color {colorUseRule=UseColor_ColorTable, specifiedColor=<rgba 0 0 0 0>}
       cornerRadii=<4 4 4 4>
      }
      tintColor {colorUseRule=UseColor_Specified, specifiedColor=<rgba 0.283 0.1607 0.0427 1>}
      uVRegion {bIsValid=false, max=<0 0>, min=<0 0>}
     }
     pressedForeground {colorUseRule=UseColor_ColorTable, specifiedColor=<rgba 0 0 0 0>}
     pressedPadding=<ltrb 0 0 0 0>
     pressedSlateSound {resourceObject=null}
    }}
   slot {class=HorizontalBoxSlot, properties {padding=<ltrb 0 0 5 0>}}
  }
  Image_1 {
   widgetClassPath=&Image
   bIsVariable=true
   parent=SubfolderTargetAMPathACButton
   widget=Image_1
   properties {brush {
     bIsDynamicallyLoaded=false
     drawAs=Image
     imageType=FullColor
     mirroring=NoMirror
     resourceName=None
     resourceObject=&AnimBatchDup_Icon_Folder
     tiling=NoTile
     imageSize=<14 14>
     margin=<ltrb 0 0 0 0>
     outlineSettings {
      bUseBrushTransparency=false
      roundingType=HalfHeightRadius
      width=0
      color {colorUseRule=UseColor_Specified, specifiedColor=<rgba 0 0 0 0>}
      cornerRadii=<0 0 0 0>
     }
     tintColor {colorUseRule=UseColor_Specified, specifiedColor=<rgba 0.1719 0.0979 0.0266 1>}
     uVRegion {bIsValid=false, max=<0 0>, min=<0 0>}
    }}
   slot {class=ButtonSlot, properties {padding=<ltrb 3 3 3 3>}}
  }
  TargetAMPathAsteriscText {
   widgetClassPath=&TextBlock
   bIsVariable=true
   parent=ACTargetAnimMontagesBox
   widget=TargetAMPathAsteriscText
   properties {
    text=*
    colorAndOpacity {colorUseRule=UseColor_Specified, specifiedColor=<rgba 0.1274 0.1274 1 1>}
    font {
     bForceMonospaced=false
     bMaterialIsStencil=false
     fontMaterial=null
     fontObject=&Roboto
     letterSpacing=0
     monospacedWidth=1
     size=9
     skewAmount=0
     typefaceFontName=Bold
     outlineSettings {
      bApplyOutlineToDropShadows=false
      bMiteredCorners=false
      bSeparateFillAlpha=false
      outlineMaterial=null
      outlineSize=0
      outlineColor=<rgba 0 0 0 1>
     }
    }
   }
   slot {class=HorizontalBoxSlot, properties {verticalAlignment=VAlign_Center}}
  }
  TargetAMPathText {
   widgetClassPath=&TextBlock
   bIsVariable=true
   parent=ACTargetAnimMontagesBox
   widget=TargetAMPathText
   properties {
    text=Text Block
    colorAndOpacity {colorUseRule=UseColor_Specified, specifiedColor=<rgba 0.1274 0.1274 1 1>}
    font {
     bForceMonospaced=false
     bMaterialIsStencil=false
     fontMaterial=null
     fontObject=&Roboto
     letterSpacing=0
     monospacedWidth=1
     size=9
     skewAmount=0
     typefaceFontName=Bold
     outlineSettings {
      bApplyOutlineToDropShadows=false
      bMiteredCorners=false
      bSeparateFillAlpha=false
      outlineMaterial=null
      outlineSize=0
      outlineColor=<rgba 0 0 0 1>
     }
    }
   }
   slot {class=HorizontalBoxSlot, properties {verticalAlignment=VAlign_Center}}
  }
  TargetAMSubfolderText {
   widgetClassPath=&TextBlock
   bIsVariable=true
   parent=ACTargetAnimMontagesBox
   widget=TargetAMSubfolderText
   properties {
    text=/Subfolder
    colorAndOpacity {colorUseRule=UseColor_Specified, specifiedColor=<rgba 0.3968 0.2346 0.0782 1>}
    font {
     bForceMonospaced=false
     bMaterialIsStencil=false
     fontMaterial=null
     fontObject=&Roboto
     letterSpacing=0
     monospacedWidth=1
     size=9
     skewAmount=0
     typefaceFontName=Bold
     outlineSettings {
      bApplyOutlineToDropShadows=false
      bMiteredCorners=false
      bSeparateFillAlpha=false
      outlineMaterial=null
      outlineSize=0
      outlineColor=<rgba 0 0 0 1>
     }
    }
   }
   slot {class=HorizontalBoxSlot, properties {verticalAlignment=VAlign_Center}}
  }
  AMPathIsEmptyText {
   widgetClassPath=&TextBlock
   bIsVariable=true
   parent=TargetPathsfromACAsset
   widget=AMPathIsEmptyText
   properties {
    autoWrapText=true
    text=* AnimMontage Path is empty in AC Asset. Using AnimSequence Paths instead.
    font {
     bForceMonospaced=false
     bMaterialIsStencil=false
     fontMaterial=null
     fontObject=&Roboto
     letterSpacing=0
     monospacedWidth=1
     size=7.5
     skewAmount=0
     typefaceFontName=Regular
     outlineSettings {
      bApplyOutlineToDropShadows=false
      bMiteredCorners=false
      bSeparateFillAlpha=false
      outlineMaterial=null
      outlineSize=0
      outlineColor=<rgba 0 0 0 1>
     }
    }
   }
   slot {class=VerticalBoxSlot}
  }
  TargetPathsVerticalBox {
   widgetClassPath=&VerticalBox
   bIsVariable=true
   parent=VerticalBox_3
   widget=TargetPathsVerticalBox
   slot {class=VerticalBoxSlot, properties {verticalAlignment=VAlign_Top}}
  }
  PathTargetLevelSequencesBox {
   widgetClassPath=&HorizontalBox
   bIsVariable=true
   parent=TargetPathsVerticalBox
   widget=PathTargetLevelSequencesBox
   slot {class=VerticalBoxSlot}
  }
  SubfolderTargetLSPathButton {
   widgetClassPath=&EditorUtilityButton
   parent=PathTargetLevelSequencesBox
   widget=SubfolderTargetLSPathButton
   properties {widgetStyle {
     clickedSlateSound {resourceObject=null}
     disabled {
      bIsDynamicallyLoaded=false
      drawAs=RoundedBox
      imageType=NoImage
      mirroring=NoMirror
      resourceName=None
      resourceObject=null
      tiling=NoTile
      imageSize=<0 0>
      margin=<ltrb 0 0 0 0>
      outlineSettings {
       bUseBrushTransparency=false
       roundingType=FixedRadius
       width=1
       color {colorUseRule=UseColor_ColorTable, specifiedColor=<rgba 0 0 0 0>}
       cornerRadii=<4 4 4 4>
      }
      tintColor {colorUseRule=UseColor_Specified, specifiedColor=<rgba 0.0395 0.0395 0.0395 1>}
      uVRegion {bIsValid=false, max=<0 0>, min=<0 0>}
     }
     disabledForeground {colorUseRule=UseColor_ColorTable, specifiedColor=<rgba 0 0 0 0>}
     hovered {
      bIsDynamicallyLoaded=false
      drawAs=RoundedBox
      imageType=NoImage
      mirroring=NoMirror
      resourceName=None
      resourceObject=null
      tiling=NoTile
      imageSize=<18 18>
      margin=<ltrb 0 0 0 0>
      outlineSettings {
       bUseBrushTransparency=false
       roundingType=FixedRadius
       width=1
       color {colorUseRule=UseColor_ColorTable, specifiedColor=<rgba 0 0 0 0>}
       cornerRadii=<4 4 4 4>
      }
      tintColor {colorUseRule=UseColor_Specified, specifiedColor=<rgba 0.599 0.3697 0.1487 1>}
      uVRegion {bIsValid=false, max=<0 0>, min=<0 0>}
     }
     hoveredForeground {colorUseRule=UseColor_ColorTable, specifiedColor=<rgba 0 0 0 0>}
     hoveredSlateSound {resourceObject=null}
     normal {
      bIsDynamicallyLoaded=false
      drawAs=RoundedBox
      imageType=NoImage
      mirroring=NoMirror
      resourceName=None
      resourceObject=null
      tiling=NoTile
      imageSize=<18 18>
      margin=<ltrb 0 0 0 0>
      outlineSettings {
       bUseBrushTransparency=false
       roundingType=FixedRadius
       width=1
       color {colorUseRule=UseColor_ColorTable, specifiedColor=<rgba 0 0 0 0>}
       cornerRadii=<4 4 4 4>
      }
      tintColor {colorUseRule=UseColor_Specified, specifiedColor=<rgba 0.3968 0.2346 0.0782 1>}
      uVRegion {bIsValid=false, max=<0 0>, min=<0 0>}
     }
     normalForeground {colorUseRule=UseColor_ColorTable, specifiedColor=<rgba 0 0 0 0>}
     normalPadding=<ltrb 0 0 0 0>
     pressed {
      bIsDynamicallyLoaded=false
      drawAs=RoundedBox
      imageType=NoImage
      mirroring=NoMirror
      resourceName=None
      resourceObject=null
      tiling=NoTile
      imageSize=<18 18>
      margin=<ltrb 0 0 0 0>
      outlineSettings {
       bUseBrushTransparency=false
       roundingType=FixedRadius
       width=1
       color {colorUseRule=UseColor_ColorTable, specifiedColor=<rgba 0 0 0 0>}
       cornerRadii=<4 4 4 4>
      }
      tintColor {colorUseRule=UseColor_Specified, specifiedColor=<rgba 0.283 0.1607 0.0427 1>}
      uVRegion {bIsValid=false, max=<0 0>, min=<0 0>}
     }
     pressedForeground {colorUseRule=UseColor_ColorTable, specifiedColor=<rgba 0 0 0 0>}
     pressedPadding=<ltrb 0 0 0 0>
     pressedSlateSound {resourceObject=null}
    }}
   slot {class=HorizontalBoxSlot, properties {padding=<ltrb 0 0 5 0>}}
  }
  Image_2 {
   widgetClassPath=&Image
   bIsVariable=true
   parent=SubfolderTargetLSPathButton
   widget=Image_2
   properties {brush {
     bIsDynamicallyLoaded=false
     drawAs=Image
     imageType=FullColor
     mirroring=NoMirror
     resourceName=None
     resourceObject=&AnimBatchDup_Icon_Folder
     tiling=NoTile
     imageSize=<14 14>
     margin=<ltrb 0 0 0 0>
     outlineSettings {
      bUseBrushTransparency=false
      roundingType=HalfHeightRadius
      width=0
      color {colorUseRule=UseColor_Specified, specifiedColor=<rgba 0 0 0 0>}
      cornerRadii=<0 0 0 0>
     }
     tintColor {colorUseRule=UseColor_Specified, specifiedColor=<rgba 0.1719 0.0979 0.0266 1>}
     uVRegion {bIsValid=false, max=<0 0>, min=<0 0>}
    }}
   slot {class=ButtonSlot, properties {padding=<ltrb 3 3 3 3>}}
  }
  TextBlock_2 {
   widgetClassPath=&TextBlock
   parent=PathTargetLevelSequencesBox
   widget=TextBlock_2
   properties {
    text=Target Level Sequences
    colorAndOpacity {colorUseRule=UseColor_Specified, specifiedColor=<rgba 0.5776 0.0802 0.0802 1>}
    font {
     bForceMonospaced=false
     bMaterialIsStencil=false
     fontMaterial=null
     fontObject=&Roboto
     letterSpacing=0
     monospacedWidth=1
     size=9
     skewAmount=0
     typefaceFontName=Bold
     outlineSettings {
      bApplyOutlineToDropShadows=false
      bMiteredCorners=false
      bSeparateFillAlpha=false
      outlineMaterial=null
      outlineSize=0
      outlineColor=<rgba 0 0 0 1>
     }
    }
   }
   slot {class=HorizontalBoxSlot, properties {verticalAlignment=VAlign_Center, padding=<ltrb 0 0 5 0>}}
  }
  TargetLSPathEditableTextBox {
   widgetClassPath=&EditorUtilityEditableTextBox
   bIsVariable=true
   parent=PathTargetLevelSequencesBox
   widget=TargetLSPathEditableTextBox
   slot {class=HorizontalBoxSlot, properties {padding=<ltrb 0 0 2 0>, size {sizeRule=Fill, value=1}}}
  }
  TargetLSSubfolderCustomPathText {
   widgetClassPath=&TextBlock
   bIsVariable=true
   parent=PathTargetLevelSequencesBox
   widget=TargetLSSubfolderCustomPathText
   properties {
    text=/Subfolder
    colorAndOpacity {colorUseRule=UseColor_Specified, specifiedColor=<rgba 0.3968 0.2346 0.0782 1>}
    font {
     bForceMonospaced=false
     bMaterialIsStencil=false
     fontMaterial=null
     fontObject=&Roboto
     letterSpacing=0
     monospacedWidth=1
     size=9
     skewAmount=0
     typefaceFontName=Bold
     outlineSettings {
      bApplyOutlineToDropShadows=false
      bMiteredCorners=false
      bSeparateFillAlpha=false
      outlineMaterial=null
      outlineSize=0
      outlineColor=<rgba 0 0 0 1>
     }
    }
   }
   slot {class=HorizontalBoxSlot, properties {verticalAlignment=VAlign_Center, padding=<ltrb 0 0 20 0>}}
  }
  PathTargetAnimSequencesBox {
   widgetClassPath=&HorizontalBox
   bIsVariable=true
   parent=TargetPathsVerticalBox
   widget=PathTargetAnimSequencesBox
   slot {class=VerticalBoxSlot}
  }
  SubfolderTargetASPathButton {
   widgetClassPath=&EditorUtilityButton
   parent=PathTargetAnimSequencesBox
   widget=SubfolderTargetASPathButton
   properties {widgetStyle {
     clickedSlateSound {resourceObject=null}
     disabled {
      bIsDynamicallyLoaded=false
      drawAs=RoundedBox
      imageType=NoImage
      mirroring=NoMirror
      resourceName=None
      resourceObject=null
      tiling=NoTile
      imageSize=<0 0>
      margin=<ltrb 0 0 0 0>
      outlineSettings {
       bUseBrushTransparency=false
       roundingType=FixedRadius
       width=1
       color {colorUseRule=UseColor_ColorTable, specifiedColor=<rgba 0 0 0 0>}
       cornerRadii=<4 4 4 4>
      }
      tintColor {colorUseRule=UseColor_Specified, specifiedColor=<rgba 0.0395 0.0395 0.0395 1>}
      uVRegion {bIsValid=false, max=<0 0>, min=<0 0>}
     }
     disabledForeground {colorUseRule=UseColor_ColorTable, specifiedColor=<rgba 0 0 0 0>}
     hovered {
      bIsDynamicallyLoaded=false
      drawAs=RoundedBox
      imageType=NoImage
      mirroring=NoMirror
      resourceName=None
      resourceObject=null
      tiling=NoTile
      imageSize=<18 18>
      margin=<ltrb 0 0 0 0>
      outlineSettings {
       bUseBrushTransparency=false
       roundingType=FixedRadius
       width=1
       color {colorUseRule=UseColor_ColorTable, specifiedColor=<rgba 0 0 0 0>}
       cornerRadii=<4 4 4 4>
      }
      tintColor {colorUseRule=UseColor_Specified, specifiedColor=<rgba 0.599 0.3697 0.1487 1>}
      uVRegion {bIsValid=false, max=<0 0>, min=<0 0>}
     }
     hoveredForeground {colorUseRule=UseColor_ColorTable, specifiedColor=<rgba 0 0 0 0>}
     hoveredSlateSound {resourceObject=null}
     normal {
      bIsDynamicallyLoaded=false
      drawAs=RoundedBox
      imageType=NoImage
      mirroring=NoMirror
      resourceName=None
      resourceObject=null
      tiling=NoTile
      imageSize=<18 18>
      margin=<ltrb 0 0 0 0>
      outlineSettings {
       bUseBrushTransparency=false
       roundingType=FixedRadius
       width=1
       color {colorUseRule=UseColor_ColorTable, specifiedColor=<rgba 0 0 0 0>}
       cornerRadii=<4 4 4 4>
      }
      tintColor {colorUseRule=UseColor_Specified, specifiedColor=<rgba 0.3968 0.2346 0.0782 1>}
      uVRegion {bIsValid=false, max=<0 0>, min=<0 0>}
     }
     normalForeground {colorUseRule=UseColor_ColorTable, specifiedColor=<rgba 0 0 0 0>}
     normalPadding=<ltrb 0 0 0 0>
     pressed {
      bIsDynamicallyLoaded=false
      drawAs=RoundedBox
      imageType=NoImage
      mirroring=NoMirror
      resourceName=None
      resourceObject=null
      tiling=NoTile
      imageSize=<18 18>
      margin=<ltrb 0 0 0 0>
      outlineSettings {
       bUseBrushTransparency=false
       roundingType=FixedRadius
       width=1
       color {colorUseRule=UseColor_ColorTable, specifiedColor=<rgba 0 0 0 0>}
       cornerRadii=<4 4 4 4>
      }
      tintColor {colorUseRule=UseColor_Specified, specifiedColor=<rgba 0.283 0.1607 0.0427 1>}
      uVRegion {bIsValid=false, max=<0 0>, min=<0 0>}
     }
     pressedForeground {colorUseRule=UseColor_ColorTable, specifiedColor=<rgba 0 0 0 0>}
     pressedPadding=<ltrb 0 0 0 0>
     pressedSlateSound {resourceObject=null}
    }}
   slot {class=HorizontalBoxSlot, properties {padding=<ltrb 0 0 5 0>}}
  }
  Image_3 {
   widgetClassPath=&Image
   bIsVariable=true
   parent=SubfolderTargetASPathButton
   widget=Image_3
   properties {brush {
     bIsDynamicallyLoaded=false
     drawAs=Image
     imageType=FullColor
     mirroring=NoMirror
     resourceName=None
     resourceObject=&AnimBatchDup_Icon_Folder
     tiling=NoTile
     imageSize=<14 14>
     margin=<ltrb 0 0 0 0>
     outlineSettings {
      bUseBrushTransparency=false
      roundingType=HalfHeightRadius
      width=0
      color {colorUseRule=UseColor_Specified, specifiedColor=<rgba 0 0 0 0>}
      cornerRadii=<0 0 0 0>
     }
     tintColor {colorUseRule=UseColor_Specified, specifiedColor=<rgba 0.1719 0.0979 0.0266 1>}
     uVRegion {bIsValid=false, max=<0 0>, min=<0 0>}
    }}
   slot {class=ButtonSlot, properties {padding=<ltrb 3 3 3 3>}}
  }
  TextBlock_7 {
   widgetClassPath=&TextBlock
   parent=PathTargetAnimSequencesBox
   widget=TextBlock_7
   properties {
    text=Target Anim Sequences
    colorAndOpacity {colorUseRule=UseColor_Specified, specifiedColor=<rgba 0.0802 0.1981 0.0648 1>}
    font {
     bForceMonospaced=false
     bMaterialIsStencil=false
     fontMaterial=null
     fontObject=&Roboto
     letterSpacing=0
     monospacedWidth=1
     size=9
     skewAmount=0
     typefaceFontName=Bold
     outlineSettings {
      bApplyOutlineToDropShadows=false
      bMiteredCorners=false
      bSeparateFillAlpha=false
      outlineMaterial=null
      outlineSize=0
      outlineColor=<rgba 0 0 0 1>
     }
    }
   }
   slot {class=HorizontalBoxSlot, properties {verticalAlignment=VAlign_Center, padding=<ltrb 0 0 5 0>}}
  }
  TargetASPathEditableTextBox {
   widgetClassPath=&EditorUtilityEditableTextBox
   bIsVariable=true
   parent=PathTargetAnimSequencesBox
   widget=TargetASPathEditableTextBox
   slot {class=HorizontalBoxSlot, properties {padding=<ltrb 0 0 2 0>, size {sizeRule=Fill, value=1}}}
  }
  TargetASSubfolderCustomPathText {
   widgetClassPath=&TextBlock
   bIsVariable=true
   parent=PathTargetAnimSequencesBox
   widget=TargetASSubfolderCustomPathText
   properties {
    text=/Subfolder
    colorAndOpacity {colorUseRule=UseColor_Specified, specifiedColor=<rgba 0.3968 0.2346 0.0782 1>}
    font {
     bForceMonospaced=false
     bMaterialIsStencil=false
     fontMaterial=null
     fontObject=&Roboto
     letterSpacing=0
     monospacedWidth=1
     size=9
     skewAmount=0
     typefaceFontName=Bold
     outlineSettings {
      bApplyOutlineToDropShadows=false
      bMiteredCorners=false
      bSeparateFillAlpha=false
      outlineMaterial=null
      outlineSize=0
      outlineColor=<rgba 0 0 0 1>
     }
    }
   }
   slot {class=HorizontalBoxSlot, properties {verticalAlignment=VAlign_Center, padding=<ltrb 0 0 20 0>}}
  }
  PathTargetAnimMontagesBox {
   widgetClassPath=&HorizontalBox
   bIsVariable=true
   parent=TargetPathsVerticalBox
   widget=PathTargetAnimMontagesBox
   slot {class=VerticalBoxSlot}
  }
  SubfolderTargetAMPathButton {
   widgetClassPath=&EditorUtilityButton
   parent=PathTargetAnimMontagesBox
   widget=SubfolderTargetAMPathButton
   properties {widgetStyle {
     clickedSlateSound {resourceObject=null}
     disabled {
      bIsDynamicallyLoaded=false
      drawAs=RoundedBox
      imageType=NoImage
      mirroring=NoMirror
      resourceName=None
      resourceObject=null
      tiling=NoTile
      imageSize=<0 0>
      margin=<ltrb 0 0 0 0>
      outlineSettings {
       bUseBrushTransparency=false
       roundingType=FixedRadius
       width=1
       color {colorUseRule=UseColor_ColorTable, specifiedColor=<rgba 0 0 0 0>}
       cornerRadii=<4 4 4 4>
      }
      tintColor {colorUseRule=UseColor_Specified, specifiedColor=<rgba 0.0395 0.0395 0.0395 1>}
      uVRegion {bIsValid=false, max=<0 0>, min=<0 0>}
     }
     disabledForeground {colorUseRule=UseColor_ColorTable, specifiedColor=<rgba 0 0 0 0>}
     hovered {
      bIsDynamicallyLoaded=false
      drawAs=RoundedBox
      imageType=NoImage
      mirroring=NoMirror
      resourceName=None
      resourceObject=null
      tiling=NoTile
      imageSize=<18 18>
      margin=<ltrb 0 0 0 0>
      outlineSettings {
       bUseBrushTransparency=false
       roundingType=FixedRadius
       width=1
       color {colorUseRule=UseColor_ColorTable, specifiedColor=<rgba 0 0 0 0>}
       cornerRadii=<4 4 4 4>
      }
      tintColor {colorUseRule=UseColor_Specified, specifiedColor=<rgba 0.599 0.3697 0.1487 1>}
      uVRegion {bIsValid=false, max=<0 0>, min=<0 0>}
     }
     hoveredForeground {colorUseRule=UseColor_ColorTable, specifiedColor=<rgba 0 0 0 0>}
     hoveredSlateSound {resourceObject=null}
     normal {
      bIsDynamicallyLoaded=false
      drawAs=RoundedBox
      imageType=NoImage
      mirroring=NoMirror
      resourceName=None
      resourceObject=null
      tiling=NoTile
      imageSize=<18 18>
      margin=<ltrb 0 0 0 0>
      outlineSettings {
       bUseBrushTransparency=false
       roundingType=FixedRadius
       width=1
       color {colorUseRule=UseColor_ColorTable, specifiedColor=<rgba 0 0 0 0>}
       cornerRadii=<4 4 4 4>
      }
      tintColor {colorUseRule=UseColor_Specified, specifiedColor=<rgba 0.3968 0.2346 0.0782 1>}
      uVRegion {bIsValid=false, max=<0 0>, min=<0 0>}
     }
     normalForeground {colorUseRule=UseColor_ColorTable, specifiedColor=<rgba 0 0 0 0>}
     normalPadding=<ltrb 0 0 0 0>
     pressed {
      bIsDynamicallyLoaded=false
      drawAs=RoundedBox
      imageType=NoImage
      mirroring=NoMirror
      resourceName=None
      resourceObject=null
      tiling=NoTile
      imageSize=<18 18>
      margin=<ltrb 0 0 0 0>
      outlineSettings {
       bUseBrushTransparency=false
       roundingType=FixedRadius
       width=1
       color {colorUseRule=UseColor_ColorTable, specifiedColor=<rgba 0 0 0 0>}
       cornerRadii=<4 4 4 4>
      }
      tintColor {colorUseRule=UseColor_Specified, specifiedColor=<rgba 0.283 0.1607 0.0427 1>}
      uVRegion {bIsValid=false, max=<0 0>, min=<0 0>}
     }
     pressedForeground {colorUseRule=UseColor_ColorTable, specifiedColor=<rgba 0 0 0 0>}
     pressedPadding=<ltrb 0 0 0 0>
     pressedSlateSound {resourceObject=null}
    }}
   slot {class=HorizontalBoxSlot, properties {padding=<ltrb 0 0 5 0>}}
  }
  Image_4 {
   widgetClassPath=&Image
   bIsVariable=true
   parent=SubfolderTargetAMPathButton
   widget=Image_4
   properties {brush {
     bIsDynamicallyLoaded=false
     drawAs=Image
     imageType=FullColor
     mirroring=NoMirror
     resourceName=None
     resourceObject=&AnimBatchDup_Icon_Folder
     tiling=NoTile
     imageSize=<14 14>
     margin=<ltrb 0 0 0 0>
     outlineSettings {
      bUseBrushTransparency=false
      roundingType=HalfHeightRadius
      width=0
      color {colorUseRule=UseColor_Specified, specifiedColor=<rgba 0 0 0 0>}
      cornerRadii=<0 0 0 0>
     }
     tintColor {colorUseRule=UseColor_Specified, specifiedColor=<rgba 0.1719 0.0979 0.0266 1>}
     uVRegion {bIsValid=false, max=<0 0>, min=<0 0>}
    }}
   slot {class=ButtonSlot, properties {padding=<ltrb 3 3 3 3>}}
  }
  TextBlock_8 {
   widgetClassPath=&TextBlock
   parent=PathTargetAnimMontagesBox
   widget=TextBlock_8
   properties {
    text=Target Anim Montages
    colorAndOpacity {colorUseRule=UseColor_Specified, specifiedColor=<rgba 0.1274 0.1274 1 1>}
    font {
     bForceMonospaced=false
     bMaterialIsStencil=false
     fontMaterial=null
     fontObject=&Roboto
     letterSpacing=0
     monospacedWidth=1
     size=9
     skewAmount=0
     typefaceFontName=Bold
     outlineSettings {
      bApplyOutlineToDropShadows=false
      bMiteredCorners=false
      bSeparateFillAlpha=false
      outlineMaterial=null
      outlineSize=0
      outlineColor=<rgba 0 0 0 1>
     }
    }
   }
   slot {class=HorizontalBoxSlot, properties {verticalAlignment=VAlign_Center, padding=<ltrb 0 0 5 0>}}
  }
  TargetAMPathEditableTextBox {
   widgetClassPath=&EditorUtilityEditableTextBox
   bIsVariable=true
   parent=PathTargetAnimMontagesBox
   widget=TargetAMPathEditableTextBox
   slot {class=HorizontalBoxSlot, properties {padding=<ltrb 0 0 2 0>, size {sizeRule=Fill, value=1}}}
  }
  TargetAMSubfolderCustomPathText {
   widgetClassPath=&TextBlock
   bIsVariable=true
   parent=PathTargetAnimMontagesBox
   widget=TargetAMSubfolderCustomPathText
   properties {
    text=/Subfolder
    colorAndOpacity {colorUseRule=UseColor_Specified, specifiedColor=<rgba 0.3968 0.2346 0.0782 1>}
    font {
     bForceMonospaced=false
     bMaterialIsStencil=false
     fontMaterial=null
     fontObject=&Roboto
     letterSpacing=0
     monospacedWidth=1
     size=9
     skewAmount=0
     typefaceFontName=Bold
     outlineSettings {
      bApplyOutlineToDropShadows=false
      bMiteredCorners=false
      bSeparateFillAlpha=false
      outlineMaterial=null
      outlineSize=0
      outlineColor=<rgba 0 0 0 1>
     }
    }
   }
   slot {class=HorizontalBoxSlot, properties {verticalAlignment=VAlign_Center, padding=<ltrb 0 0 20 0>}}
  }
  Border_175 {
   widgetClassPath=&Border
   parent=VerticalBox_1
   widget=Border_175
   properties {brushColor=<rgba 0 0 0 1>, padding=<ltrb 0 2 0 0>}
   slot {class=VerticalBoxSlot, properties {padding=<ltrb 10 0 10 0>}}
  }
  RenamingVerticalBox {
   widgetClassPath=&VerticalBox
   parent=VerticalBox_1
   widget=RenamingVerticalBox
   properties {bOverride_Cursor=true}
   slot {class=VerticalBoxSlot, properties {padding=<ltrb 15 10 15 10>}}
  }
  VerticalBox_82 {
   widgetClassPath=&VerticalBox
   parent=RenamingVerticalBox
   widget=VerticalBox_82
   slot {class=VerticalBoxSlot, properties {verticalAlignment=VAlign_Center}}
  }
  HorizontalBox_210 {
   widgetClassPath=&HorizontalBox
   parent=VerticalBox_82
   widget=HorizontalBox_210
   slot {class=VerticalBoxSlot, properties {padding=<ltrb 0 0 0 5>}}
  }
  TextBlock_216 {
   widgetClassPath=&TextBlock
   parent=HorizontalBox_210
   widget=TextBlock_216
   properties {
    text=Rename Target Animations
    font {
     bForceMonospaced=false
     bMaterialIsStencil=false
     fontMaterial=null
     fontObject=&Roboto
     letterSpacing=0
     monospacedWidth=1
     size=10.5
     skewAmount=0
     typefaceFontName=Bold
     outlineSettings {
      bApplyOutlineToDropShadows=false
      bMiteredCorners=false
      bSeparateFillAlpha=false
      outlineMaterial=null
      outlineSize=0
      outlineColor=<rgba 0 0 0 1>
     }
    }
   }
   slot {class=HorizontalBoxSlot, properties {padding=<ltrb 0 0 10 0>}}
  }
  RenameAnimTypeCheckbox {
   widgetClassPath=&EditorUtilityCheckBox
   bIsVariable=true
   parent=HorizontalBox_210
   widget=RenameAnimTypeCheckbox
   properties {widgetStyle {
     checkBoxType=CheckBox
     backgroundHoveredImage {
      bIsDynamicallyLoaded=false
      drawAs=RoundedBox
      imageType=NoImage
      mirroring=NoMirror
      resourceName=None
      resourceObject=null
      tiling=NoTile
      imageSize=<18 18>
      margin=<ltrb 0 0 0 0>
      outlineSettings {
       bUseBrushTransparency=false
       roundingType=FixedRadius
       width=1
       color {colorUseRule=UseColor_ColorTable, specifiedColor=<rgba 0 0 0 0>}
       cornerRadii=<3 3 3 3>
      }
      tintColor {colorUseRule=UseColor_Specified, specifiedColor=<rgba 0.0048 0.0048 0.0048 1>}
      uVRegion {bIsValid=false, max=<0 0>, min=<0 0>}
     }
     backgroundImage {
      bIsDynamicallyLoaded=false
      drawAs=RoundedBox
      imageType=NoImage
      mirroring=NoMirror
      resourceName=None
      resourceObject=null
      tiling=NoTile
      imageSize=<18 18>
      margin=<ltrb 0 0 0 0>
      outlineSettings {
       bUseBrushTransparency=false
       roundingType=FixedRadius
       width=1
       color {colorUseRule=UseColor_ColorTable, specifiedColor=<rgba 0 0 0 0>}
       cornerRadii=<3 3 3 3>
      }
      tintColor {colorUseRule=UseColor_Specified, specifiedColor=<rgba 0.0048 0.0048 0.0048 1>}
      uVRegion {bIsValid=false, max=<0 0>, min=<0 0>}
     }
     backgroundPressedImage {
      bIsDynamicallyLoaded=false
      drawAs=RoundedBox
      imageType=NoImage
      mirroring=NoMirror
      resourceName=None
      resourceObject=null
      tiling=NoTile
      imageSize=<18 18>
      margin=<ltrb 0 0 0 0>
      outlineSettings {
       bUseBrushTransparency=false
       roundingType=FixedRadius
       width=1
       color {colorUseRule=UseColor_ColorTable, specifiedColor=<rgba 0 0 0 0>}
       cornerRadii=<3 3 3 3>
      }
      tintColor {colorUseRule=UseColor_Specified, specifiedColor=<rgba 0.0048 0.0048 0.0048 1>}
      uVRegion {bIsValid=false, max=<0 0>, min=<0 0>}
     }
     borderBackgroundColor {colorUseRule=UseColor_Specified, specifiedColor=<rgba 1 1 1 1>}
     checkedForeground {colorUseRule=UseColor_Specified, specifiedColor=<rgba 1 1 1 1>}
     checkedHoveredForeground {colorUseRule=UseColor_Specified, specifiedColor=<rgba 1 1 1 1>}
     checkedHoveredImage {
      bIsDynamicallyLoaded=false
      drawAs=Image
      imageType=Vector
      mirroring=NoMirror
      resourceName=../../../Engine/Content/Slate/Starship/CoreWidgets/CheckBox/check.svg
      resourceObject=null
      tiling=NoTile
      imageSize=<16 16>
      margin=<ltrb 0 0 0 0>
      outlineSettings {
       bUseBrushTransparency=false
       roundingType=HalfHeightRadius
       width=0
       color {colorUseRule=UseColor_Specified, specifiedColor=<rgba 0 0 0 0>}
       cornerRadii=<0 0 0 0>
      }
      tintColor {colorUseRule=UseColor_Specified, specifiedColor=<rgba 0.0044 0.2384 1 1>}
      uVRegion {bIsValid=false, max=<0 0>, min=<0 0>}
     }
     checkedImage {
      bIsDynamicallyLoaded=false
      drawAs=Image
      imageType=Vector
      mirroring=NoMirror
      resourceName=../../../Engine/Content/Slate/Starship/CoreWidgets/CheckBox/check.svg
      resourceObject=null
      tiling=NoTile
      imageSize=<16 16>
      margin=<ltrb 0 0 0 0>
      outlineSettings {
       bUseBrushTransparency=false
       roundingType=HalfHeightRadius
       width=0
       color {colorUseRule=UseColor_Specified, specifiedColor=<rgba 0 0 0 0>}
       cornerRadii=<0 0 0 0>
      }
      tintColor {colorUseRule=UseColor_Specified, specifiedColor=<rgba 0 0.162 0.7454 1>}
      uVRegion {bIsValid=false, max=<0 0>, min=<0 0>}
     }
     checkedPressedForeground {colorUseRule=UseColor_Specified, specifiedColor=<rgba 1 1 1 1>}
     checkedPressedImage {
      bIsDynamicallyLoaded=false
      drawAs=Image
      imageType=Vector
      mirroring=NoMirror
      resourceName=../../../Engine/Content/Slate/Starship/CoreWidgets/CheckBox/check.svg
      resourceObject=null
      tiling=NoTile
      imageSize=<16 16>
      margin=<ltrb 0 0 0 0>
      outlineSettings {
       bUseBrushTransparency=false
       roundingType=HalfHeightRadius
       width=0
       color {colorUseRule=UseColor_Specified, specifiedColor=<rgba 0 0 0 0>}
       cornerRadii=<0 0 0 0>
      }
      tintColor {colorUseRule=UseColor_Specified, specifiedColor=<rgba 0 0.0802 0.3515 1>}
      uVRegion {bIsValid=false, max=<0 0>, min=<0 0>}
     }
     checkedSlateSound {resourceObject=null}
     foregroundColor {colorUseRule=UseColor_Specified, specifiedColor=<rgba 1 1 1 1>}
     hoveredForeground {colorUseRule=UseColor_Specified, specifiedColor=<rgba 1 1 1 1>}
     hoveredSlateSound {resourceObject=null}
     padding=<ltrb 0 0 0 0>
     pressedForeground {colorUseRule=UseColor_Specified, specifiedColor=<rgba 1 1 1 1>}
     uncheckedHoveredImage {
      bIsDynamicallyLoaded=false
      drawAs=NoDrawType
      imageType=NoImage
      mirroring=NoMirror
      resourceName=None
      resourceObject=null
      tiling=NoTile
      imageSize=<0 0>
      margin=<ltrb 0 0 0 0>
      outlineSettings {
       bUseBrushTransparency=false
       roundingType=HalfHeightRadius
       width=0
       color {colorUseRule=UseColor_Specified, specifiedColor=<rgba 0 0 0 0>}
       cornerRadii=<0 0 0 0>
      }
      tintColor {colorUseRule=UseColor_Specified, specifiedColor=<rgba 1 1 1 1>}
      uVRegion {bIsValid=false, max=<0 0>, min=<0 0>}
     }
     uncheckedImage {
      bIsDynamicallyLoaded=false
      drawAs=NoDrawType
      imageType=NoImage
      mirroring=NoMirror
      resourceName=None
      resourceObject=null
      tiling=NoTile
      imageSize=<0 0>
      margin=<ltrb 0 0 0 0>
      outlineSettings {
       bUseBrushTransparency=false
       roundingType=HalfHeightRadius
       width=0
       color {colorUseRule=UseColor_Specified, specifiedColor=<rgba 0 0 0 0>}
       cornerRadii=<0 0 0 0>
      }
      tintColor {colorUseRule=UseColor_Specified, specifiedColor=<rgba 1 1 1 1>}
      uVRegion {bIsValid=false, max=<0 0>, min=<0 0>}
     }
     uncheckedPressedImage {
      bIsDynamicallyLoaded=false
      drawAs=NoDrawType
      imageType=NoImage
      mirroring=NoMirror
      resourceName=None
      resourceObject=null
      tiling=NoTile
      imageSize=<0 0>
      margin=<ltrb 0 0 0 0>
      outlineSettings {
       bUseBrushTransparency=false
       roundingType=HalfHeightRadius
       width=0
       color {colorUseRule=UseColor_Specified, specifiedColor=<rgba 0 0 0 0>}
       cornerRadii=<0 0 0 0>
      }
      tintColor {colorUseRule=UseColor_Specified, specifiedColor=<rgba 1 1 1 1>}
      uVRegion {bIsValid=false, max=<0 0>, min=<0 0>}
     }
     uncheckedSlateSound {resourceObject=null}
     undeterminedForeground {colorUseRule=UseColor_Specified, specifiedColor=<rgba 1 1 1 1>}
     undeterminedHoveredImage {
      bIsDynamicallyLoaded=false
      drawAs=Image
      imageType=Vector
      mirroring=NoMirror
      resourceName=../../../Engine/Content/Slate/Starship/CoreWidgets/CheckBox/indeterminate.svg
      resourceObject=null
      tiling=NoTile
      imageSize=<16 16>
      margin=<ltrb 0 0 0 0>
      outlineSettings {
       bUseBrushTransparency=false
       roundingType=HalfHeightRadius
       width=0
       color {colorUseRule=UseColor_Specified, specifiedColor=<rgba 0 0 0 0>}
       cornerRadii=<0 0 0 0>
      }
      tintColor {colorUseRule=UseColor_Specified, specifiedColor=<rgba 0.0044 0.2384 1 1>}
      uVRegion {bIsValid=false, max=<0 0>, min=<0 0>}
     }
     undeterminedImage {
      bIsDynamicallyLoaded=false
      drawAs=Image
      imageType=Vector
      mirroring=NoMirror
      resourceName=../../../Engine/Content/Slate/Starship/CoreWidgets/CheckBox/indeterminate.svg
      resourceObject=null
      tiling=NoTile
      imageSize=<16 16>
      margin=<ltrb 0 0 0 0>
      outlineSettings {
       bUseBrushTransparency=false
       roundingType=HalfHeightRadius
       width=0
       color {colorUseRule=UseColor_Specified, specifiedColor=<rgba 0 0 0 0>}
       cornerRadii=<0 0 0 0>
      }
      tintColor {colorUseRule=UseColor_Specified, specifiedColor=<rgba 0 0.162 0.7454 1>}
      uVRegion {bIsValid=false, max=<0 0>, min=<0 0>}
     }
     undeterminedPressedImage {
      bIsDynamicallyLoaded=false
      drawAs=Image
      imageType=Vector
      mirroring=NoMirror
      resourceName=../../../Engine/Content/Slate/Starship/CoreWidgets/CheckBox/indeterminate.svg
      resourceObject=null
      tiling=NoTile
      imageSize=<16 16>
      margin=<ltrb 0 0 0 0>
      outlineSettings {
       bUseBrushTransparency=false
       roundingType=HalfHeightRadius
       width=0
       color {colorUseRule=UseColor_Specified, specifiedColor=<rgba 0 0 0 0>}
       cornerRadii=<0 0 0 0>
      }
      tintColor {colorUseRule=UseColor_Specified, specifiedColor=<rgba 0 0.0802 0.3515 1>}
      uVRegion {bIsValid=false, max=<0 0>, min=<0 0>}
     }
    }}
   slot {class=HorizontalBoxSlot, properties {horizontalAlignment=HAlign_Center, verticalAlignment=VAlign_Center, padding=<ltrb 0 0 3 0>}}
  }
  TextBlock_312 {
   widgetClassPath=&TextBlock
   parent=HorizontalBox_210
   widget=TextBlock_312
   properties {
    text=Rename per animation type
    font {
     bForceMonospaced=false
     bMaterialIsStencil=false
     fontMaterial=null
     fontObject=&Roboto
     letterSpacing=0
     monospacedWidth=1
     size=7.5
     skewAmount=0
     typefaceFontName=Regular
     outlineSettings {
      bApplyOutlineToDropShadows=false
      bMiteredCorners=false
      bSeparateFillAlpha=false
      outlineMaterial=null
      outlineSize=0
      outlineColor=<rgba 0 0 0 1>
     }
    }
   }
   slot {class=HorizontalBoxSlot, properties {verticalAlignment=VAlign_Center}}
  }
  RenameLevelSequenceHorizontalBox {widgetClassPath=&HorizontalBox, parent=RenamingVerticalBox, widget=RenameLevelSequenceHorizontalBox, slot {class=VerticalBoxSlot}}
  HorizontalBox_208 {
   widgetClassPath=&HorizontalBox
   parent=RenameLevelSequenceHorizontalBox
   widget=HorizontalBox_208
   properties {bOverride_Cursor=true}
   slot {class=HorizontalBoxSlot, properties {padding=<ltrb 0 0 10 0>, size {sizeRule=Fill, value=1}}}
  }
  FromLSText {
   widgetClassPath=&TextBlock
   bIsVariable=true
   parent=HorizontalBox_208
   widget=FromLSText
   properties {
    text=From
    font {
     bForceMonospaced=false
     bMaterialIsStencil=false
     fontMaterial=null
     fontObject=&Roboto
     letterSpacing=0
     monospacedWidth=1
     size=9
     skewAmount=0
     typefaceFontName=Bold
     outlineSettings {
      bApplyOutlineToDropShadows=false
      bMiteredCorners=false
      bSeparateFillAlpha=false
      outlineMaterial=null
      outlineSize=0
      outlineColor=<rgba 0 0 0 1>
     }
    }
   }
   slot {class=HorizontalBoxSlot, properties {verticalAlignment=VAlign_Center, padding=<ltrb 0 0 5 0>}}
  }
  ReplaceLSFromEditableText {
   widgetClassPath=&EditorUtilityEditableTextBox
   bIsVariable=true
   parent=HorizontalBox_208
   widget=ReplaceLSFromEditableText
   properties {widgetStyle {
     backgroundColor {colorUseRule=UseColor_Specified, specifiedColor=<rgba 1 1 1 1>}
     backgroundImageFocused {
      bIsDynamicallyLoaded=false
      drawAs=RoundedBox
      imageType=NoImage
      mirroring=NoMirror
      resourceName=None
      resourceObject=null
      tiling=NoTile
      imageSize=<0 0>
      margin=<ltrb 0 0 0 0>
      outlineSettings {
       bUseBrushTransparency=false
       roundingType=FixedRadius
       width=1
       color {colorUseRule=UseColor_ColorTable, specifiedColor=<rgba 0 0 0 0>}
       cornerRadii=<4 4 4 4>
      }
      tintColor {colorUseRule=UseColor_Specified, specifiedColor=<rgba 0.0048 0.0048 0.0048 1>}
      uVRegion {bIsValid=false, max=<0 0>, min=<0 0>}
     }
     backgroundImageHovered {
      bIsDynamicallyLoaded=false
      drawAs=RoundedBox
      imageType=NoImage
      mirroring=NoMirror
      resourceName=None
      resourceObject=null
      tiling=NoTile
      imageSize=<0 0>
      margin=<ltrb 0 0 0 0>
      outlineSettings {
       bUseBrushTransparency=false
       roundingType=FixedRadius
       width=1
       color {colorUseRule=UseColor_ColorTable, specifiedColor=<rgba 0 0 0 0>}
       cornerRadii=<4 4 4 4>
      }
      tintColor {colorUseRule=UseColor_Specified, specifiedColor=<rgba 0.0048 0.0048 0.0048 1>}
      uVRegion {bIsValid=false, max=<0 0>, min=<0 0>}
     }
     backgroundImageNormal {
      bIsDynamicallyLoaded=false
      drawAs=RoundedBox
      imageType=NoImage
      mirroring=NoMirror
      resourceName=None
      resourceObject=null
      tiling=NoTile
      imageSize=<0 0>
      margin=<ltrb 0 0 0 0>
      outlineSettings {
       bUseBrushTransparency=false
       roundingType=FixedRadius
       width=1
       color {colorUseRule=UseColor_ColorTable, specifiedColor=<rgba 0 0 0 0>}
       cornerRadii=<4 4 4 4>
      }
      tintColor {colorUseRule=UseColor_Specified, specifiedColor=<rgba 0.0048 0.0048 0.0048 1>}
      uVRegion {bIsValid=false, max=<0 0>, min=<0 0>}
     }
     backgroundImageReadOnly {
      bIsDynamicallyLoaded=false
      drawAs=RoundedBox
      imageType=NoImage
      mirroring=NoMirror
      resourceName=None
      resourceObject=null
      tiling=NoTile
      imageSize=<0 0>
      margin=<ltrb 0 0 0 0>
      outlineSettings {
       bUseBrushTransparency=false
       roundingType=FixedRadius
       width=1
       color {colorUseRule=UseColor_ColorTable, specifiedColor=<rgba 0 0 0 0>}
       cornerRadii=<4 4 4 4>
      }
      tintColor {colorUseRule=UseColor_Specified, specifiedColor=<rgba 0.0284 0.0284 0.0284 1>}
      uVRegion {bIsValid=false, max=<0 0>, min=<0 0>}
     }
     focusedForegroundColor {colorUseRule=UseColor_Specified, specifiedColor=<rgba 1 1 1 1>}
     foregroundColor {colorUseRule=UseColor_Specified, specifiedColor=<rgba 0.5271 0.5271 0.5271 1>}
     hScrollBarPadding=<ltrb 0 0 0 0>
     padding=<ltrb 8 3 8 4>
     readOnlyForegroundColor {colorUseRule=UseColor_Foreground, specifiedColor=<rgba 1 0 1 1>}
     scrollBarStyle {
      thickness=8
      draggedThumbImage {
       bIsDynamicallyLoaded=false
       drawAs=RoundedBox
       imageType=NoImage
       mirroring=NoMirror
       resourceName=None
       resourceObject=null
       tiling=NoTile
       imageSize=<0 0>
       margin=<ltrb 0 0 0 0>
       outlineSettings {
        bUseBrushTransparency=false
        roundingType=FixedRadius
        width=0
        color {colorUseRule=UseColor_Specified, specifiedColor=<rgba 0 0 0 0>}
        cornerRadii=<4 4 4 4>
       }
       tintColor {colorUseRule=UseColor_Specified, specifiedColor=<rgba 0.2159 0.2159 0.2159 1>}
       uVRegion {bIsValid=false, max=<0 0>, min=<0 0>}
      }
      horizontalBackgroundImage {
       bIsDynamicallyLoaded=false
       drawAs=NoDrawType
       imageType=NoImage
       mirroring=NoMirror
       resourceName=None
       resourceObject=null
       tiling=NoTile
       imageSize=<0 0>
       margin=<ltrb 0 0 0 0>
       outlineSettings {
        bUseBrushTransparency=false
        roundingType=HalfHeightRadius
        width=0
        color {colorUseRule=UseColor_Specified, specifiedColor=<rgba 0 0 0 0>}
        cornerRadii=<0 0 0 0>
       }
       tintColor {colorUseRule=UseColor_Specified, specifiedColor=<rgba 1 1 1 1>}
       uVRegion {bIsValid=false, max=<0 0>, min=<0 0>}
      }
      horizontalBottomSlotImage {
       bIsDynamicallyLoaded=false
       drawAs=NoDrawType
       imageType=NoImage
       mirroring=NoMirror
       resourceName=None
       resourceObject=null
       tiling=NoTile
       imageSize=<0 0>
       margin=<ltrb 0 0 0 0>
       outlineSettings {
        bUseBrushTransparency=false
        roundingType=HalfHeightRadius
        width=0
        color {colorUseRule=UseColor_Specified, specifiedColor=<rgba 0 0 0 0>}
        cornerRadii=<0 0 0 0>
       }
       tintColor {colorUseRule=UseColor_Specified, specifiedColor=<rgba 1 1 1 1>}
       uVRegion {bIsValid=false, max=<0 0>, min=<0 0>}
      }
      horizontalTopSlotImage {
       bIsDynamicallyLoaded=false
       drawAs=NoDrawType
       imageType=NoImage
       mirroring=NoMirror
       resourceName=None
       resourceObject=null
       tiling=NoTile
       imageSize=<0 0>
       margin=<ltrb 0 0 0 0>
       outlineSettings {
        bUseBrushTransparency=false
        roundingType=HalfHeightRadius
        width=0
        color {colorUseRule=UseColor_Specified, specifiedColor=<rgba 0 0 0 0>}
        cornerRadii=<0 0 0 0>
       }
       tintColor {colorUseRule=UseColor_Specified, specifiedColor=<rgba 1 1 1 1>}
       uVRegion {bIsValid=false, max=<0 0>, min=<0 0>}
      }
      hoveredThumbImage {
       bIsDynamicallyLoaded=false
       drawAs=RoundedBox
       imageType=NoImage
       mirroring=NoMirror
       resourceName=None
       resourceObject=null
       tiling=NoTile
       imageSize=<0 0>
       margin=<ltrb 0 0 0 0>
       outlineSettings {
        bUseBrushTransparency=false
        roundingType=FixedRadius
        width=0
        color {colorUseRule=UseColor_Specified, specifiedColor=<rgba 0 0 0 0>}
        cornerRadii=<4 4 4 4>
       }
       tintColor {colorUseRule=UseColor_Specified, specifiedColor=<rgba 0.2159 0.2159 0.2159 1>}
       uVRegion {bIsValid=false, max=<0 0>, min=<0 0>}
      }
      normalThumbImage {
       bIsDynamicallyLoaded=false
       drawAs=RoundedBox
       imageType=NoImage
       mirroring=NoMirror
       resourceName=None
       resourceObject=null
       tiling=NoTile
       imageSize=<0 0>
       margin=<ltrb 0 0 0 0>
       outlineSettings {
        bUseBrushTransparency=false
        roundingType=FixedRadius
        width=0
        color {colorUseRule=UseColor_Specified, specifiedColor=<rgba 0 0 0 0>}
        cornerRadii=<4 4 4 4>
       }
       tintColor {colorUseRule=UseColor_Specified, specifiedColor=<rgba 0.0953 0.0953 0.0953 1>}
       uVRegion {bIsValid=false, max=<0 0>, min=<0 0>}
      }
      verticalBackgroundImage {
       bIsDynamicallyLoaded=false
       drawAs=NoDrawType
       imageType=NoImage
       mirroring=NoMirror
       resourceName=None
       resourceObject=null
       tiling=NoTile
       imageSize=<0 0>
       margin=<ltrb 0 0 0 0>
       outlineSettings {
        bUseBrushTransparency=false
        roundingType=HalfHeightRadius
        width=0
        color {colorUseRule=UseColor_Specified, specifiedColor=<rgba 0 0 0 0>}
        cornerRadii=<0 0 0 0>
       }
       tintColor {colorUseRule=UseColor_Specified, specifiedColor=<rgba 1 1 1 1>}
       uVRegion {bIsValid=false, max=<0 0>, min=<0 0>}
      }
      verticalBottomSlotImage {
       bIsDynamicallyLoaded=false
       drawAs=NoDrawType
       imageType=NoImage
       mirroring=NoMirror
       resourceName=None
       resourceObject=null
       tiling=NoTile
       imageSize=<0 0>
       margin=<ltrb 0 0 0 0>
       outlineSettings {
        bUseBrushTransparency=false
        roundingType=HalfHeightRadius
        width=0
        color {colorUseRule=UseColor_Specified, specifiedColor=<rgba 0 0 0 0>}
        cornerRadii=<0 0 0 0>
       }
       tintColor {colorUseRule=UseColor_Specified, specifiedColor=<rgba 1 1 1 1>}
       uVRegion {bIsValid=false, max=<0 0>, min=<0 0>}
      }
      verticalTopSlotImage {
       bIsDynamicallyLoaded=false
       drawAs=NoDrawType
       imageType=NoImage
       mirroring=NoMirror
       resourceName=None
       resourceObject=null
       tiling=NoTile
       imageSize=<0 0>
       margin=<ltrb 0 0 0 0>
       outlineSettings {
        bUseBrushTransparency=false
        roundingType=HalfHeightRadius
        width=0
        color {colorUseRule=UseColor_Specified, specifiedColor=<rgba 0 0 0 0>}
        cornerRadii=<0 0 0 0>
       }
       tintColor {colorUseRule=UseColor_Specified, specifiedColor=<rgba 1 1 1 1>}
       uVRegion {bIsValid=false, max=<0 0>, min=<0 0>}
      }
     }
     textStyle {
      overflowPolicy=Clip
      transformPolicy=None
      colorAndOpacity {colorUseRule=UseColor_Foreground, specifiedColor=<rgba 1 0 1 1>}
      font {
       bForceMonospaced=false
       bMaterialIsStencil=false
       fontMaterial=null
       fontObject=null
       letterSpacing=0
       monospacedWidth=1
       size=9
       skewAmount=0
       typefaceFontName=Regular
       outlineSettings {
        bApplyOutlineToDropShadows=false
        bMiteredCorners=false
        bSeparateFillAlpha=false
        outlineMaterial=null
        outlineSize=0
        outlineColor=<rgba 0 0 0 1>
       }
      }
      highlightColor {colorUseRule=UseColor_ColorTable, specifiedColor=<rgba 0 0 0 0>}
      highlightShape {
       bIsDynamicallyLoaded=false
       drawAs=Image
       imageType=NoImage
       mirroring=NoMirror
       resourceName=None
       resourceObject=null
       tiling=NoTile
       imageSize=<0 0>
       margin=<ltrb 0 0 0 0>
       outlineSettings {
        bUseBrushTransparency=false
        roundingType=HalfHeightRadius
        width=0
        color {colorUseRule=UseColor_Specified, specifiedColor=<rgba 0 0 0 0>}
        cornerRadii=<0 0 0 0>
       }
       tintColor {colorUseRule=UseColor_ColorTable, specifiedColor=<rgba 0 0 0 0>}
       uVRegion {bIsValid=false, max=<0 0>, min=<0 0>}
      }
      selectedBackgroundColor {colorUseRule=UseColor_ColorTable, specifiedColor=<rgba 0 0 0 0>}
      shadowColorAndOpacity=<rgba 0 0 0 1>
      shadowOffset=<0 0>
      strikeBrush {
       bIsDynamicallyLoaded=false
       drawAs=Image
       imageType=NoImage
       mirroring=NoMirror
       resourceName=None
       resourceObject=null
       tiling=NoTile
       imageSize=<32 32>
       margin=<ltrb 0 0 0 0>
       outlineSettings {
        bUseBrushTransparency=false
        roundingType=HalfHeightRadius
        width=0
        color {colorUseRule=UseColor_Specified, specifiedColor=<rgba 0 0 0 0>}
        cornerRadii=<0 0 0 0>
       }
       tintColor {colorUseRule=UseColor_Specified, specifiedColor=<rgba 1 1 1 1>}
       uVRegion {bIsValid=false, max=<0 0>, min=<0 0>}
      }
      underlineBrush {
       bIsDynamicallyLoaded=false
       drawAs=Image
       imageType=NoImage
       mirroring=NoMirror
       resourceName=None
       resourceObject=null
       tiling=NoTile
       imageSize=<32 32>
       margin=<ltrb 0 0 0 0>
       outlineSettings {
        bUseBrushTransparency=false
        roundingType=HalfHeightRadius
        width=0
        color {colorUseRule=UseColor_Specified, specifiedColor=<rgba 0 0 0 0>}
        cornerRadii=<0 0 0 0>
       }
       tintColor {colorUseRule=UseColor_Specified, specifiedColor=<rgba 1 1 1 1>}
       uVRegion {bIsValid=false, max=<0 0>, min=<0 0>}
      }
     }
     vScrollBarPadding=<ltrb 0 0 0 0>
    }}
   slot {class=HorizontalBoxSlot, properties {size {sizeRule=Fill, value=1}}}
  }
  HorizontalBox_1 {
   widgetClassPath=&HorizontalBox
   parent=RenameLevelSequenceHorizontalBox
   widget=HorizontalBox_1
   properties {bOverride_Cursor=true}
   slot {class=HorizontalBoxSlot, properties {padding=<ltrb 10 0 0 0>, size {sizeRule=Fill, value=1}}}
  }
  ToLSText {
   widgetClassPath=&TextBlock
   bIsVariable=true
   parent=HorizontalBox_1
   widget=ToLSText
   properties {
    text=To
    font {
     bForceMonospaced=false
     bMaterialIsStencil=false
     fontMaterial=null
     fontObject=&Roboto
     letterSpacing=0
     monospacedWidth=1
     size=9
     skewAmount=0
     typefaceFontName=Bold
     outlineSettings {
      bApplyOutlineToDropShadows=false
      bMiteredCorners=false
      bSeparateFillAlpha=false
      outlineMaterial=null
      outlineSize=0
      outlineColor=<rgba 0 0 0 1>
     }
    }
   }
   slot {class=HorizontalBoxSlot, properties {verticalAlignment=VAlign_Center, padding=<ltrb 0 0 5 0>}}
  }
  ReplaceLSToEditableText {
   widgetClassPath=&EditorUtilityEditableTextBox
   bIsVariable=true
   parent=HorizontalBox_1
   widget=ReplaceLSToEditableText
   properties {widgetStyle {
     backgroundColor {colorUseRule=UseColor_Specified, specifiedColor=<rgba 1 1 1 1>}
     backgroundImageFocused {
      bIsDynamicallyLoaded=false
      drawAs=RoundedBox
      imageType=NoImage
      mirroring=NoMirror
      resourceName=None
      resourceObject=null
      tiling=NoTile
      imageSize=<0 0>
      margin=<ltrb 0 0 0 0>
      outlineSettings {
       bUseBrushTransparency=false
       roundingType=FixedRadius
       width=1
       color {colorUseRule=UseColor_ColorTable, specifiedColor=<rgba 0 0 0 0>}
       cornerRadii=<4 4 4 4>
      }
      tintColor {colorUseRule=UseColor_Specified, specifiedColor=<rgba 0.0048 0.0048 0.0048 1>}
      uVRegion {bIsValid=false, max=<0 0>, min=<0 0>}
     }
     backgroundImageHovered {
      bIsDynamicallyLoaded=false
      drawAs=RoundedBox
      imageType=NoImage
      mirroring=NoMirror
      resourceName=None
      resourceObject=null
      tiling=NoTile
      imageSize=<0 0>
      margin=<ltrb 0 0 0 0>
      outlineSettings {
       bUseBrushTransparency=false
       roundingType=FixedRadius
       width=1
       color {colorUseRule=UseColor_ColorTable, specifiedColor=<rgba 0 0 0 0>}
       cornerRadii=<4 4 4 4>
      }
      tintColor {colorUseRule=UseColor_Specified, specifiedColor=<rgba 0.0048 0.0048 0.0048 1>}
      uVRegion {bIsValid=false, max=<0 0>, min=<0 0>}
     }
     backgroundImageNormal {
      bIsDynamicallyLoaded=false
      drawAs=RoundedBox
      imageType=NoImage
      mirroring=NoMirror
      resourceName=None
      resourceObject=null
      tiling=NoTile
      imageSize=<0 0>
      margin=<ltrb 0 0 0 0>
      outlineSettings {
       bUseBrushTransparency=false
       roundingType=FixedRadius
       width=1
       color {colorUseRule=UseColor_ColorTable, specifiedColor=<rgba 0 0 0 0>}
       cornerRadii=<4 4 4 4>
      }
      tintColor {colorUseRule=UseColor_Specified, specifiedColor=<rgba 0.0048 0.0048 0.0048 1>}
      uVRegion {bIsValid=false, max=<0 0>, min=<0 0>}
     }
     backgroundImageReadOnly {
      bIsDynamicallyLoaded=false
      drawAs=RoundedBox
      imageType=NoImage
      mirroring=NoMirror
      resourceName=None
      resourceObject=null
      tiling=NoTile
      imageSize=<0 0>
      margin=<ltrb 0 0 0 0>
      outlineSettings {
       bUseBrushTransparency=false
       roundingType=FixedRadius
       width=1
       color {colorUseRule=UseColor_ColorTable, specifiedColor=<rgba 0 0 0 0>}
       cornerRadii=<4 4 4 4>
      }
      tintColor {colorUseRule=UseColor_Specified, specifiedColor=<rgba 0.0284 0.0284 0.0284 1>}
      uVRegion {bIsValid=false, max=<0 0>, min=<0 0>}
     }
     focusedForegroundColor {colorUseRule=UseColor_Specified, specifiedColor=<rgba 1 1 1 1>}
     foregroundColor {colorUseRule=UseColor_Specified, specifiedColor=<rgba 0.5271 0.5271 0.5271 1>}
     hScrollBarPadding=<ltrb 0 0 0 0>
     padding=<ltrb 8 3 8 4>
     readOnlyForegroundColor {colorUseRule=UseColor_Foreground, specifiedColor=<rgba 1 0 1 1>}
     scrollBarStyle {
      thickness=8
      draggedThumbImage {
       bIsDynamicallyLoaded=false
       drawAs=RoundedBox
       imageType=NoImage
       mirroring=NoMirror
       resourceName=None
       resourceObject=null
       tiling=NoTile
       imageSize=<0 0>
       margin=<ltrb 0 0 0 0>
       outlineSettings {
        bUseBrushTransparency=false
        roundingType=FixedRadius
        width=0
        color {colorUseRule=UseColor_Specified, specifiedColor=<rgba 0 0 0 0>}
        cornerRadii=<4 4 4 4>
       }
       tintColor {colorUseRule=UseColor_Specified, specifiedColor=<rgba 0.2159 0.2159 0.2159 1>}
       uVRegion {bIsValid=false, max=<0 0>, min=<0 0>}
      }
      horizontalBackgroundImage {
       bIsDynamicallyLoaded=false
       drawAs=NoDrawType
       imageType=NoImage
       mirroring=NoMirror
       resourceName=None
       resourceObject=null
       tiling=NoTile
       imageSize=<0 0>
       margin=<ltrb 0 0 0 0>
       outlineSettings {
        bUseBrushTransparency=false
        roundingType=HalfHeightRadius
        width=0
        color {colorUseRule=UseColor_Specified, specifiedColor=<rgba 0 0 0 0>}
        cornerRadii=<0 0 0 0>
       }
       tintColor {colorUseRule=UseColor_Specified, specifiedColor=<rgba 1 1 1 1>}
       uVRegion {bIsValid=false, max=<0 0>, min=<0 0>}
      }
      horizontalBottomSlotImage {
       bIsDynamicallyLoaded=false
       drawAs=NoDrawType
       imageType=NoImage
       mirroring=NoMirror
       resourceName=None
       resourceObject=null
       tiling=NoTile
       imageSize=<0 0>
       margin=<ltrb 0 0 0 0>
       outlineSettings {
        bUseBrushTransparency=false
        roundingType=HalfHeightRadius
        width=0
        color {colorUseRule=UseColor_Specified, specifiedColor=<rgba 0 0 0 0>}
        cornerRadii=<0 0 0 0>
       }
       tintColor {colorUseRule=UseColor_Specified, specifiedColor=<rgba 1 1 1 1>}
       uVRegion {bIsValid=false, max=<0 0>, min=<0 0>}
      }
      horizontalTopSlotImage {
       bIsDynamicallyLoaded=false
       drawAs=NoDrawType
       imageType=NoImage
       mirroring=NoMirror
       resourceName=None
       resourceObject=null
       tiling=NoTile
       imageSize=<0 0>
       margin=<ltrb 0 0 0 0>
       outlineSettings {
        bUseBrushTransparency=false
        roundingType=HalfHeightRadius
        width=0
        color {colorUseRule=UseColor_Specified, specifiedColor=<rgba 0 0 0 0>}
        cornerRadii=<0 0 0 0>
       }
       tintColor {colorUseRule=UseColor_Specified, specifiedColor=<rgba 1 1 1 1>}
       uVRegion {bIsValid=false, max=<0 0>, min=<0 0>}
      }
      hoveredThumbImage {
       bIsDynamicallyLoaded=false
       drawAs=RoundedBox
       imageType=NoImage
       mirroring=NoMirror
       resourceName=None
       resourceObject=null
       tiling=NoTile
       imageSize=<0 0>
       margin=<ltrb 0 0 0 0>
       outlineSettings {
        bUseBrushTransparency=false
        roundingType=FixedRadius
        width=0
        color {colorUseRule=UseColor_Specified, specifiedColor=<rgba 0 0 0 0>}
        cornerRadii=<4 4 4 4>
       }
       tintColor {colorUseRule=UseColor_Specified, specifiedColor=<rgba 0.2159 0.2159 0.2159 1>}
       uVRegion {bIsValid=false, max=<0 0>, min=<0 0>}
      }
      normalThumbImage {
       bIsDynamicallyLoaded=false
       drawAs=RoundedBox
       imageType=NoImage
       mirroring=NoMirror
       resourceName=None
       resourceObject=null
       tiling=NoTile
       imageSize=<0 0>
       margin=<ltrb 0 0 0 0>
       outlineSettings {
        bUseBrushTransparency=false
        roundingType=FixedRadius
        width=0
        color {colorUseRule=UseColor_Specified, specifiedColor=<rgba 0 0 0 0>}
        cornerRadii=<4 4 4 4>
       }
       tintColor {colorUseRule=UseColor_Specified, specifiedColor=<rgba 0.0953 0.0953 0.0953 1>}
       uVRegion {bIsValid=false, max=<0 0>, min=<0 0>}
      }
      verticalBackgroundImage {
       bIsDynamicallyLoaded=false
       drawAs=NoDrawType
       imageType=NoImage
       mirroring=NoMirror
       resourceName=None
       resourceObject=null
       tiling=NoTile
       imageSize=<0 0>
       margin=<ltrb 0 0 0 0>
       outlineSettings {
        bUseBrushTransparency=false
        roundingType=HalfHeightRadius
        width=0
        color {colorUseRule=UseColor_Specified, specifiedColor=<rgba 0 0 0 0>}
        cornerRadii=<0 0 0 0>
       }
       tintColor {colorUseRule=UseColor_Specified, specifiedColor=<rgba 1 1 1 1>}
       uVRegion {bIsValid=false, max=<0 0>, min=<0 0>}
      }
      verticalBottomSlotImage {
       bIsDynamicallyLoaded=false
       drawAs=NoDrawType
       imageType=NoImage
       mirroring=NoMirror
       resourceName=None
       resourceObject=null
       tiling=NoTile
       imageSize=<0 0>
       margin=<ltrb 0 0 0 0>
       outlineSettings {
        bUseBrushTransparency=false
        roundingType=HalfHeightRadius
        width=0
        color {colorUseRule=UseColor_Specified, specifiedColor=<rgba 0 0 0 0>}
        cornerRadii=<0 0 0 0>
       }
       tintColor {colorUseRule=UseColor_Specified, specifiedColor=<rgba 1 1 1 1>}
       uVRegion {bIsValid=false, max=<0 0>, min=<0 0>}
      }
      verticalTopSlotImage {
       bIsDynamicallyLoaded=false
       drawAs=NoDrawType
       imageType=NoImage
       mirroring=NoMirror
       resourceName=None
       resourceObject=null
       tiling=NoTile
       imageSize=<0 0>
       margin=<ltrb 0 0 0 0>
       outlineSettings {
        bUseBrushTransparency=false
        roundingType=HalfHeightRadius
        width=0
        color {colorUseRule=UseColor_Specified, specifiedColor=<rgba 0 0 0 0>}
        cornerRadii=<0 0 0 0>
       }
       tintColor {colorUseRule=UseColor_Specified, specifiedColor=<rgba 1 1 1 1>}
       uVRegion {bIsValid=false, max=<0 0>, min=<0 0>}
      }
     }
     textStyle {
      overflowPolicy=Clip
      transformPolicy=None
      colorAndOpacity {colorUseRule=UseColor_Foreground, specifiedColor=<rgba 1 0 1 1>}
      font {
       bForceMonospaced=false
       bMaterialIsStencil=false
       fontMaterial=null
       fontObject=null
       letterSpacing=0
       monospacedWidth=1
       size=9
       skewAmount=0
       typefaceFontName=Regular
       outlineSettings {
        bApplyOutlineToDropShadows=false
        bMiteredCorners=false
        bSeparateFillAlpha=false
        outlineMaterial=null
        outlineSize=0
        outlineColor=<rgba 0 0 0 1>
       }
      }
      highlightColor {colorUseRule=UseColor_ColorTable, specifiedColor=<rgba 0 0 0 0>}
      highlightShape {
       bIsDynamicallyLoaded=false
       drawAs=Image
       imageType=NoImage
       mirroring=NoMirror
       resourceName=None
       resourceObject=null
       tiling=NoTile
       imageSize=<0 0>
       margin=<ltrb 0 0 0 0>
       outlineSettings {
        bUseBrushTransparency=false
        roundingType=HalfHeightRadius
        width=0
        color {colorUseRule=UseColor_Specified, specifiedColor=<rgba 0 0 0 0>}
        cornerRadii=<0 0 0 0>
       }
       tintColor {colorUseRule=UseColor_ColorTable, specifiedColor=<rgba 0 0 0 0>}
       uVRegion {bIsValid=false, max=<0 0>, min=<0 0>}
      }
      selectedBackgroundColor {colorUseRule=UseColor_ColorTable, specifiedColor=<rgba 0 0 0 0>}
      shadowColorAndOpacity=<rgba 0 0 0 1>
      shadowOffset=<0 0>
      strikeBrush {
       bIsDynamicallyLoaded=false
       drawAs=Image
       imageType=NoImage
       mirroring=NoMirror
       resourceName=None
       resourceObject=null
       tiling=NoTile
       imageSize=<32 32>
       margin=<ltrb 0 0 0 0>
       outlineSettings {
        bUseBrushTransparency=false
        roundingType=HalfHeightRadius
        width=0
        color {colorUseRule=UseColor_Specified, specifiedColor=<rgba 0 0 0 0>}
        cornerRadii=<0 0 0 0>
       }
       tintColor {colorUseRule=UseColor_Specified, specifiedColor=<rgba 1 1 1 1>}
       uVRegion {bIsValid=false, max=<0 0>, min=<0 0>}
      }
      underlineBrush {
       bIsDynamicallyLoaded=false
       drawAs=Image
       imageType=NoImage
       mirroring=NoMirror
       resourceName=None
       resourceObject=null
       tiling=NoTile
       imageSize=<32 32>
       margin=<ltrb 0 0 0 0>
       outlineSettings {
        bUseBrushTransparency=false
        roundingType=HalfHeightRadius
        width=0
        color {colorUseRule=UseColor_Specified, specifiedColor=<rgba 0 0 0 0>}
        cornerRadii=<0 0 0 0>
       }
       tintColor {colorUseRule=UseColor_Specified, specifiedColor=<rgba 1 1 1 1>}
       uVRegion {bIsValid=false, max=<0 0>, min=<0 0>}
      }
     }
     vScrollBarPadding=<ltrb 0 0 0 0>
    }}
   slot {class=HorizontalBoxSlot, properties {size {sizeRule=Fill, value=1}}}
  }
  RenameAnimSequenceHorizontalBox {
   widgetClassPath=&HorizontalBox
   bIsVariable=true
   parent=RenamingVerticalBox
   widget=RenameAnimSequenceHorizontalBox
   slot {class=VerticalBoxSlot}
  }
  HorizontalBox_2 {
   widgetClassPath=&HorizontalBox
   parent=RenameAnimSequenceHorizontalBox
   widget=HorizontalBox_2
   properties {bOverride_Cursor=true}
   slot {class=HorizontalBoxSlot, properties {padding=<ltrb 0 0 10 0>, size {sizeRule=Fill, value=1}}}
  }
  TextBlock_3 {
   widgetClassPath=&TextBlock
   parent=HorizontalBox_2
   widget=TextBlock_3
   properties {
    text=From
    colorAndOpacity {colorUseRule=UseColor_Specified, specifiedColor=<rgba 0.0802 0.1981 0.0648 1>}
    font {
     bForceMonospaced=false
     bMaterialIsStencil=false
     fontMaterial=null
     fontObject=&Roboto
     letterSpacing=0
     monospacedWidth=1
     size=9
     skewAmount=0
     typefaceFontName=Bold
     outlineSettings {
      bApplyOutlineToDropShadows=false
      bMiteredCorners=false
      bSeparateFillAlpha=false
      outlineMaterial=null
      outlineSize=0
      outlineColor=<rgba 0 0 0 1>
     }
    }
   }
   slot {class=HorizontalBoxSlot, properties {verticalAlignment=VAlign_Center, padding=<ltrb 0 0 5 0>}}
  }
  ReplaceASFromEditableText {
   widgetClassPath=&EditorUtilityEditableTextBox
   bIsVariable=true
   parent=HorizontalBox_2
   widget=ReplaceASFromEditableText
   properties {widgetStyle {
     backgroundColor {colorUseRule=UseColor_Specified, specifiedColor=<rgba 1 1 1 1>}
     backgroundImageFocused {
      bIsDynamicallyLoaded=false
      drawAs=RoundedBox
      imageType=NoImage
      mirroring=NoMirror
      resourceName=None
      resourceObject=null
      tiling=NoTile
      imageSize=<0 0>
      margin=<ltrb 0 0 0 0>
      outlineSettings {
       bUseBrushTransparency=false
       roundingType=FixedRadius
       width=1
       color {colorUseRule=UseColor_ColorTable, specifiedColor=<rgba 0 0 0 0>}
       cornerRadii=<4 4 4 4>
      }
      tintColor {colorUseRule=UseColor_Specified, specifiedColor=<rgba 0.0048 0.0048 0.0048 1>}
      uVRegion {bIsValid=false, max=<0 0>, min=<0 0>}
     }
     backgroundImageHovered {
      bIsDynamicallyLoaded=false
      drawAs=RoundedBox
      imageType=NoImage
      mirroring=NoMirror
      resourceName=None
      resourceObject=null
      tiling=NoTile
      imageSize=<0 0>
      margin=<ltrb 0 0 0 0>
      outlineSettings {
       bUseBrushTransparency=false
       roundingType=FixedRadius
       width=1
       color {colorUseRule=UseColor_ColorTable, specifiedColor=<rgba 0 0 0 0>}
       cornerRadii=<4 4 4 4>
      }
      tintColor {colorUseRule=UseColor_Specified, specifiedColor=<rgba 0.0048 0.0048 0.0048 1>}
      uVRegion {bIsValid=false, max=<0 0>, min=<0 0>}
     }
     backgroundImageNormal {
      bIsDynamicallyLoaded=false
      drawAs=RoundedBox
      imageType=NoImage
      mirroring=NoMirror
      resourceName=None
      resourceObject=null
      tiling=NoTile
      imageSize=<0 0>
      margin=<ltrb 0 0 0 0>
      outlineSettings {
       bUseBrushTransparency=false
       roundingType=FixedRadius
       width=1
       color {colorUseRule=UseColor_ColorTable, specifiedColor=<rgba 0 0 0 0>}
       cornerRadii=<4 4 4 4>
      }
      tintColor {colorUseRule=UseColor_Specified, specifiedColor=<rgba 0.0048 0.0048 0.0048 1>}
      uVRegion {bIsValid=false, max=<0 0>, min=<0 0>}
     }
     backgroundImageReadOnly {
      bIsDynamicallyLoaded=false
      drawAs=RoundedBox
      imageType=NoImage
      mirroring=NoMirror
      resourceName=None
      resourceObject=null
      tiling=NoTile
      imageSize=<0 0>
      margin=<ltrb 0 0 0 0>
      outlineSettings {
       bUseBrushTransparency=false
       roundingType=FixedRadius
       width=1
       color {colorUseRule=UseColor_ColorTable, specifiedColor=<rgba 0 0 0 0>}
       cornerRadii=<4 4 4 4>
      }
      tintColor {colorUseRule=UseColor_Specified, specifiedColor=<rgba 0.0284 0.0284 0.0284 1>}
      uVRegion {bIsValid=false, max=<0 0>, min=<0 0>}
     }
     focusedForegroundColor {colorUseRule=UseColor_Specified, specifiedColor=<rgba 1 1 1 1>}
     foregroundColor {colorUseRule=UseColor_Specified, specifiedColor=<rgba 0.5271 0.5271 0.5271 1>}
     hScrollBarPadding=<ltrb 0 0 0 0>
     padding=<ltrb 8 3 8 4>
     readOnlyForegroundColor {colorUseRule=UseColor_Foreground, specifiedColor=<rgba 1 0 1 1>}
     scrollBarStyle {
      thickness=8
      draggedThumbImage {
       bIsDynamicallyLoaded=false
       drawAs=RoundedBox
       imageType=NoImage
       mirroring=NoMirror
       resourceName=None
       resourceObject=null
       tiling=NoTile
       imageSize=<0 0>
       margin=<ltrb 0 0 0 0>
       outlineSettings {
        bUseBrushTransparency=false
        roundingType=FixedRadius
        width=0
        color {colorUseRule=UseColor_Specified, specifiedColor=<rgba 0 0 0 0>}
        cornerRadii=<4 4 4 4>
       }
       tintColor {colorUseRule=UseColor_Specified, specifiedColor=<rgba 0.2159 0.2159 0.2159 1>}
       uVRegion {bIsValid=false, max=<0 0>, min=<0 0>}
      }
      horizontalBackgroundImage {
       bIsDynamicallyLoaded=false
       drawAs=NoDrawType
       imageType=NoImage
       mirroring=NoMirror
       resourceName=None
       resourceObject=null
       tiling=NoTile
       imageSize=<0 0>
       margin=<ltrb 0 0 0 0>
       outlineSettings {
        bUseBrushTransparency=false
        roundingType=HalfHeightRadius
        width=0
        color {colorUseRule=UseColor_Specified, specifiedColor=<rgba 0 0 0 0>}
        cornerRadii=<0 0 0 0>
       }
       tintColor {colorUseRule=UseColor_Specified, specifiedColor=<rgba 1 1 1 1>}
       uVRegion {bIsValid=false, max=<0 0>, min=<0 0>}
      }
      horizontalBottomSlotImage {
       bIsDynamicallyLoaded=false
       drawAs=NoDrawType
       imageType=NoImage
       mirroring=NoMirror
       resourceName=None
       resourceObject=null
       tiling=NoTile
       imageSize=<0 0>
       margin=<ltrb 0 0 0 0>
       outlineSettings {
        bUseBrushTransparency=false
        roundingType=HalfHeightRadius
        width=0
        color {colorUseRule=UseColor_Specified, specifiedColor=<rgba 0 0 0 0>}
        cornerRadii=<0 0 0 0>
       }
       tintColor {colorUseRule=UseColor_Specified, specifiedColor=<rgba 1 1 1 1>}
       uVRegion {bIsValid=false, max=<0 0>, min=<0 0>}
      }
      horizontalTopSlotImage {
       bIsDynamicallyLoaded=false
       drawAs=NoDrawType
       imageType=NoImage
       mirroring=NoMirror
       resourceName=None
       resourceObject=null
       tiling=NoTile
       imageSize=<0 0>
       margin=<ltrb 0 0 0 0>
       outlineSettings {
        bUseBrushTransparency=false
        roundingType=HalfHeightRadius
        width=0
        color {colorUseRule=UseColor_Specified, specifiedColor=<rgba 0 0 0 0>}
        cornerRadii=<0 0 0 0>
       }
       tintColor {colorUseRule=UseColor_Specified, specifiedColor=<rgba 1 1 1 1>}
       uVRegion {bIsValid=false, max=<0 0>, min=<0 0>}
      }
      hoveredThumbImage {
       bIsDynamicallyLoaded=false
       drawAs=RoundedBox
       imageType=NoImage
       mirroring=NoMirror
       resourceName=None
       resourceObject=null
       tiling=NoTile
       imageSize=<0 0>
       margin=<ltrb 0 0 0 0>
       outlineSettings {
        bUseBrushTransparency=false
        roundingType=FixedRadius
        width=0
        color {colorUseRule=UseColor_Specified, specifiedColor=<rgba 0 0 0 0>}
        cornerRadii=<4 4 4 4>
       }
       tintColor {colorUseRule=UseColor_Specified, specifiedColor=<rgba 0.2159 0.2159 0.2159 1>}
       uVRegion {bIsValid=false, max=<0 0>, min=<0 0>}
      }
      normalThumbImage {
       bIsDynamicallyLoaded=false
       drawAs=RoundedBox
       imageType=NoImage
       mirroring=NoMirror
       resourceName=None
       resourceObject=null
       tiling=NoTile
       imageSize=<0 0>
       margin=<ltrb 0 0 0 0>
       outlineSettings {
        bUseBrushTransparency=false
        roundingType=FixedRadius
        width=0
        color {colorUseRule=UseColor_Specified, specifiedColor=<rgba 0 0 0 0>}
        cornerRadii=<4 4 4 4>
       }
       tintColor {colorUseRule=UseColor_Specified, specifiedColor=<rgba 0.0953 0.0953 0.0953 1>}
       uVRegion {bIsValid=false, max=<0 0>, min=<0 0>}
      }
      verticalBackgroundImage {
       bIsDynamicallyLoaded=false
       drawAs=NoDrawType
       imageType=NoImage
       mirroring=NoMirror
       resourceName=None
       resourceObject=null
       tiling=NoTile
       imageSize=<0 0>
       margin=<ltrb 0 0 0 0>
       outlineSettings {
        bUseBrushTransparency=false
        roundingType=HalfHeightRadius
        width=0
        color {colorUseRule=UseColor_Specified, specifiedColor=<rgba 0 0 0 0>}
        cornerRadii=<0 0 0 0>
       }
       tintColor {colorUseRule=UseColor_Specified, specifiedColor=<rgba 1 1 1 1>}
       uVRegion {bIsValid=false, max=<0 0>, min=<0 0>}
      }
      verticalBottomSlotImage {
       bIsDynamicallyLoaded=false
       drawAs=NoDrawType
       imageType=NoImage
       mirroring=NoMirror
       resourceName=None
       resourceObject=null
       tiling=NoTile
       imageSize=<0 0>
       margin=<ltrb 0 0 0 0>
       outlineSettings {
        bUseBrushTransparency=false
        roundingType=HalfHeightRadius
        width=0
        color {colorUseRule=UseColor_Specified, specifiedColor=<rgba 0 0 0 0>}
        cornerRadii=<0 0 0 0>
       }
       tintColor {colorUseRule=UseColor_Specified, specifiedColor=<rgba 1 1 1 1>}
       uVRegion {bIsValid=false, max=<0 0>, min=<0 0>}
      }
      verticalTopSlotImage {
       bIsDynamicallyLoaded=false
       drawAs=NoDrawType
       imageType=NoImage
       mirroring=NoMirror
       resourceName=None
       resourceObject=null
       tiling=NoTile
       imageSize=<0 0>
       margin=<ltrb 0 0 0 0>
       outlineSettings {
        bUseBrushTransparency=false
        roundingType=HalfHeightRadius
        width=0
        color {colorUseRule=UseColor_Specified, specifiedColor=<rgba 0 0 0 0>}
        cornerRadii=<0 0 0 0>
       }
       tintColor {colorUseRule=UseColor_Specified, specifiedColor=<rgba 1 1 1 1>}
       uVRegion {bIsValid=false, max=<0 0>, min=<0 0>}
      }
     }
     textStyle {
      overflowPolicy=Clip
      transformPolicy=None
      colorAndOpacity {colorUseRule=UseColor_Foreground, specifiedColor=<rgba 1 0 1 1>}
      font {
       bForceMonospaced=false
       bMaterialIsStencil=false
       fontMaterial=null
       fontObject=null
       letterSpacing=0
       monospacedWidth=1
       size=9
       skewAmount=0
       typefaceFontName=Regular
       outlineSettings {
        bApplyOutlineToDropShadows=false
        bMiteredCorners=false
        bSeparateFillAlpha=false
        outlineMaterial=null
        outlineSize=0
        outlineColor=<rgba 0 0 0 1>
       }
      }
      highlightColor {colorUseRule=UseColor_ColorTable, specifiedColor=<rgba 0 0 0 0>}
      highlightShape {
       bIsDynamicallyLoaded=false
       drawAs=Image
       imageType=NoImage
       mirroring=NoMirror
       resourceName=None
       resourceObject=null
       tiling=NoTile
       imageSize=<0 0>
       margin=<ltrb 0 0 0 0>
       outlineSettings {
        bUseBrushTransparency=false
        roundingType=HalfHeightRadius
        width=0
        color {colorUseRule=UseColor_Specified, specifiedColor=<rgba 0 0 0 0>}
        cornerRadii=<0 0 0 0>
       }
       tintColor {colorUseRule=UseColor_ColorTable, specifiedColor=<rgba 0 0 0 0>}
       uVRegion {bIsValid=false, max=<0 0>, min=<0 0>}
      }
      selectedBackgroundColor {colorUseRule=UseColor_ColorTable, specifiedColor=<rgba 0 0 0 0>}
      shadowColorAndOpacity=<rgba 0 0 0 1>
      shadowOffset=<0 0>
      strikeBrush {
       bIsDynamicallyLoaded=false
       drawAs=Image
       imageType=NoImage
       mirroring=NoMirror
       resourceName=None
       resourceObject=null
       tiling=NoTile
       imageSize=<32 32>
       margin=<ltrb 0 0 0 0>
       outlineSettings {
        bUseBrushTransparency=false
        roundingType=HalfHeightRadius
        width=0
        color {colorUseRule=UseColor_Specified, specifiedColor=<rgba 0 0 0 0>}
        cornerRadii=<0 0 0 0>
       }
       tintColor {colorUseRule=UseColor_Specified, specifiedColor=<rgba 1 1 1 1>}
       uVRegion {bIsValid=false, max=<0 0>, min=<0 0>}
      }
      underlineBrush {
       bIsDynamicallyLoaded=false
       drawAs=Image
       imageType=NoImage
       mirroring=NoMirror
       resourceName=None
       resourceObject=null
       tiling=NoTile
       imageSize=<32 32>
       margin=<ltrb 0 0 0 0>
       outlineSettings {
        bUseBrushTransparency=false
        roundingType=HalfHeightRadius
        width=0
        color {colorUseRule=UseColor_Specified, specifiedColor=<rgba 0 0 0 0>}
        cornerRadii=<0 0 0 0>
       }
       tintColor {colorUseRule=UseColor_Specified, specifiedColor=<rgba 1 1 1 1>}
       uVRegion {bIsValid=false, max=<0 0>, min=<0 0>}
      }
     }
     vScrollBarPadding=<ltrb 0 0 0 0>
    }}
   slot {class=HorizontalBoxSlot, properties {size {sizeRule=Fill, value=1}}}
  }
  HorizontalBox_5 {
   widgetClassPath=&HorizontalBox
   parent=RenameAnimSequenceHorizontalBox
   widget=HorizontalBox_5
   properties {bOverride_Cursor=true}
   slot {class=HorizontalBoxSlot, properties {padding=<ltrb 10 0 0 0>, size {sizeRule=Fill, value=1}}}
  }
  TextBlock_4 {
   widgetClassPath=&TextBlock
   parent=HorizontalBox_5
   widget=TextBlock_4
   properties {
    text=To
    colorAndOpacity {colorUseRule=UseColor_Specified, specifiedColor=<rgba 0.0802 0.1981 0.0648 1>}
    font {
     bForceMonospaced=false
     bMaterialIsStencil=false
     fontMaterial=null
     fontObject=&Roboto
     letterSpacing=0
     monospacedWidth=1
     size=9
     skewAmount=0
     typefaceFontName=Bold
     outlineSettings {
      bApplyOutlineToDropShadows=false
      bMiteredCorners=false
      bSeparateFillAlpha=false
      outlineMaterial=null
      outlineSize=0
      outlineColor=<rgba 0 0 0 1>
     }
    }
   }
   slot {class=HorizontalBoxSlot, properties {verticalAlignment=VAlign_Center, padding=<ltrb 0 0 5 0>}}
  }
  ReplaceASToEditableText {
   widgetClassPath=&EditorUtilityEditableTextBox
   bIsVariable=true
   parent=HorizontalBox_5
   widget=ReplaceASToEditableText
   properties {widgetStyle {
     backgroundColor {colorUseRule=UseColor_Specified, specifiedColor=<rgba 1 1 1 1>}
     backgroundImageFocused {
      bIsDynamicallyLoaded=false
      drawAs=RoundedBox
      imageType=NoImage
      mirroring=NoMirror
      resourceName=None
      resourceObject=null
      tiling=NoTile
      imageSize=<0 0>
      margin=<ltrb 0 0 0 0>
      outlineSettings {
       bUseBrushTransparency=false
       roundingType=FixedRadius
       width=1
       color {colorUseRule=UseColor_ColorTable, specifiedColor=<rgba 0 0 0 0>}
       cornerRadii=<4 4 4 4>
      }
      tintColor {colorUseRule=UseColor_Specified, specifiedColor=<rgba 0.0048 0.0048 0.0048 1>}
      uVRegion {bIsValid=false, max=<0 0>, min=<0 0>}
     }
     backgroundImageHovered {
      bIsDynamicallyLoaded=false
      drawAs=RoundedBox
      imageType=NoImage
      mirroring=NoMirror
      resourceName=None
      resourceObject=null
      tiling=NoTile
      imageSize=<0 0>
      margin=<ltrb 0 0 0 0>
      outlineSettings {
       bUseBrushTransparency=false
       roundingType=FixedRadius
       width=1
       color {colorUseRule=UseColor_ColorTable, specifiedColor=<rgba 0 0 0 0>}
       cornerRadii=<4 4 4 4>
      }
      tintColor {colorUseRule=UseColor_Specified, specifiedColor=<rgba 0.0048 0.0048 0.0048 1>}
      uVRegion {bIsValid=false, max=<0 0>, min=<0 0>}
     }
     backgroundImageNormal {
      bIsDynamicallyLoaded=false
      drawAs=RoundedBox
      imageType=NoImage
      mirroring=NoMirror
      resourceName=None
      resourceObject=null
      tiling=NoTile
      imageSize=<0 0>
      margin=<ltrb 0 0 0 0>
      outlineSettings {
       bUseBrushTransparency=false
       roundingType=FixedRadius
       width=1
       color {colorUseRule=UseColor_ColorTable, specifiedColor=<rgba 0 0 0 0>}
       cornerRadii=<4 4 4 4>
      }
      tintColor {colorUseRule=UseColor_Specified, specifiedColor=<rgba 0.0048 0.0048 0.0048 1>}
      uVRegion {bIsValid=false, max=<0 0>, min=<0 0>}
     }
     backgroundImageReadOnly {
      bIsDynamicallyLoaded=false
      drawAs=RoundedBox
      imageType=NoImage
      mirroring=NoMirror
      resourceName=None
      resourceObject=null
      tiling=NoTile
      imageSize=<0 0>
      margin=<ltrb 0 0 0 0>
      outlineSettings {
       bUseBrushTransparency=false
       roundingType=FixedRadius
       width=1
       color {colorUseRule=UseColor_ColorTable, specifiedColor=<rgba 0 0 0 0>}
       cornerRadii=<4 4 4 4>
      }
      tintColor {colorUseRule=UseColor_Specified, specifiedColor=<rgba 0.0284 0.0284 0.0284 1>}
      uVRegion {bIsValid=false, max=<0 0>, min=<0 0>}
     }
     focusedForegroundColor {colorUseRule=UseColor_Specified, specifiedColor=<rgba 1 1 1 1>}
     foregroundColor {colorUseRule=UseColor_Specified, specifiedColor=<rgba 0.5271 0.5271 0.5271 1>}
     hScrollBarPadding=<ltrb 0 0 0 0>
     padding=<ltrb 8 3 8 4>
     readOnlyForegroundColor {colorUseRule=UseColor_Foreground, specifiedColor=<rgba 1 0 1 1>}
     scrollBarStyle {
      thickness=8
      draggedThumbImage {
       bIsDynamicallyLoaded=false
       drawAs=RoundedBox
       imageType=NoImage
       mirroring=NoMirror
       resourceName=None
       resourceObject=null
       tiling=NoTile
       imageSize=<0 0>
       margin=<ltrb 0 0 0 0>
       outlineSettings {
        bUseBrushTransparency=false
        roundingType=FixedRadius
        width=0
        color {colorUseRule=UseColor_Specified, specifiedColor=<rgba 0 0 0 0>}
        cornerRadii=<4 4 4 4>
       }
       tintColor {colorUseRule=UseColor_Specified, specifiedColor=<rgba 0.2159 0.2159 0.2159 1>}
       uVRegion {bIsValid=false, max=<0 0>, min=<0 0>}
      }
      horizontalBackgroundImage {
       bIsDynamicallyLoaded=false
       drawAs=NoDrawType
       imageType=NoImage
       mirroring=NoMirror
       resourceName=None
       resourceObject=null
       tiling=NoTile
       imageSize=<0 0>
       margin=<ltrb 0 0 0 0>
       outlineSettings {
        bUseBrushTransparency=false
        roundingType=HalfHeightRadius
        width=0
        color {colorUseRule=UseColor_Specified, specifiedColor=<rgba 0 0 0 0>}
        cornerRadii=<0 0 0 0>
       }
       tintColor {colorUseRule=UseColor_Specified, specifiedColor=<rgba 1 1 1 1>}
       uVRegion {bIsValid=false, max=<0 0>, min=<0 0>}
      }
      horizontalBottomSlotImage {
       bIsDynamicallyLoaded=false
       drawAs=NoDrawType
       imageType=NoImage
       mirroring=NoMirror
       resourceName=None
       resourceObject=null
       tiling=NoTile
       imageSize=<0 0>
       margin=<ltrb 0 0 0 0>
       outlineSettings {
        bUseBrushTransparency=false
        roundingType=HalfHeightRadius
        width=0
        color {colorUseRule=UseColor_Specified, specifiedColor=<rgba 0 0 0 0>}
        cornerRadii=<0 0 0 0>
       }
       tintColor {colorUseRule=UseColor_Specified, specifiedColor=<rgba 1 1 1 1>}
       uVRegion {bIsValid=false, max=<0 0>, min=<0 0>}
      }
      horizontalTopSlotImage {
       bIsDynamicallyLoaded=false
       drawAs=NoDrawType
       imageType=NoImage
       mirroring=NoMirror
       resourceName=None
       resourceObject=null
       tiling=NoTile
       imageSize=<0 0>
       margin=<ltrb 0 0 0 0>
       outlineSettings {
        bUseBrushTransparency=false
        roundingType=HalfHeightRadius
        width=0
        color {colorUseRule=UseColor_Specified, specifiedColor=<rgba 0 0 0 0>}
        cornerRadii=<0 0 0 0>
       }
       tintColor {colorUseRule=UseColor_Specified, specifiedColor=<rgba 1 1 1 1>}
       uVRegion {bIsValid=false, max=<0 0>, min=<0 0>}
      }
      hoveredThumbImage {
       bIsDynamicallyLoaded=false
       drawAs=RoundedBox
       imageType=NoImage
       mirroring=NoMirror
       resourceName=None
       resourceObject=null
       tiling=NoTile
       imageSize=<0 0>
       margin=<ltrb 0 0 0 0>
       outlineSettings {
        bUseBrushTransparency=false
        roundingType=FixedRadius
        width=0
        color {colorUseRule=UseColor_Specified, specifiedColor=<rgba 0 0 0 0>}
        cornerRadii=<4 4 4 4>
       }
       tintColor {colorUseRule=UseColor_Specified, specifiedColor=<rgba 0.2159 0.2159 0.2159 1>}
       uVRegion {bIsValid=false, max=<0 0>, min=<0 0>}
      }
      normalThumbImage {
       bIsDynamicallyLoaded=false
       drawAs=RoundedBox
       imageType=NoImage
       mirroring=NoMirror
       resourceName=None
       resourceObject=null
       tiling=NoTile
       imageSize=<0 0>
       margin=<ltrb 0 0 0 0>
       outlineSettings {
        bUseBrushTransparency=false
        roundingType=FixedRadius
        width=0
        color {colorUseRule=UseColor_Specified, specifiedColor=<rgba 0 0 0 0>}
        cornerRadii=<4 4 4 4>
       }
       tintColor {colorUseRule=UseColor_Specified, specifiedColor=<rgba 0.0953 0.0953 0.0953 1>}
       uVRegion {bIsValid=false, max=<0 0>, min=<0 0>}
      }
      verticalBackgroundImage {
       bIsDynamicallyLoaded=false
       drawAs=NoDrawType
       imageType=NoImage
       mirroring=NoMirror
       resourceName=None
       resourceObject=null
       tiling=NoTile
       imageSize=<0 0>
       margin=<ltrb 0 0 0 0>
       outlineSettings {
        bUseBrushTransparency=false
        roundingType=HalfHeightRadius
        width=0
        color {colorUseRule=UseColor_Specified, specifiedColor=<rgba 0 0 0 0>}
        cornerRadii=<0 0 0 0>
       }
       tintColor {colorUseRule=UseColor_Specified, specifiedColor=<rgba 1 1 1 1>}
       uVRegion {bIsValid=false, max=<0 0>, min=<0 0>}
      }
      verticalBottomSlotImage {
       bIsDynamicallyLoaded=false
       drawAs=NoDrawType
       imageType=NoImage
       mirroring=NoMirror
       resourceName=None
       resourceObject=null
       tiling=NoTile
       imageSize=<0 0>
       margin=<ltrb 0 0 0 0>
       outlineSettings {
        bUseBrushTransparency=false
        roundingType=HalfHeightRadius
        width=0
        color {colorUseRule=UseColor_Specified, specifiedColor=<rgba 0 0 0 0>}
        cornerRadii=<0 0 0 0>
       }
       tintColor {colorUseRule=UseColor_Specified, specifiedColor=<rgba 1 1 1 1>}
       uVRegion {bIsValid=false, max=<0 0>, min=<0 0>}
      }
      verticalTopSlotImage {
       bIsDynamicallyLoaded=false
       drawAs=NoDrawType
       imageType=NoImage
       mirroring=NoMirror
       resourceName=None
       resourceObject=null
       tiling=NoTile
       imageSize=<0 0>
       margin=<ltrb 0 0 0 0>
       outlineSettings {
        bUseBrushTransparency=false
        roundingType=HalfHeightRadius
        width=0
        color {colorUseRule=UseColor_Specified, specifiedColor=<rgba 0 0 0 0>}
        cornerRadii=<0 0 0 0>
       }
       tintColor {colorUseRule=UseColor_Specified, specifiedColor=<rgba 1 1 1 1>}
       uVRegion {bIsValid=false, max=<0 0>, min=<0 0>}
      }
     }
     textStyle {
      overflowPolicy=Clip
      transformPolicy=None
      colorAndOpacity {colorUseRule=UseColor_Foreground, specifiedColor=<rgba 1 0 1 1>}
      font {
       bForceMonospaced=false
       bMaterialIsStencil=false
       fontMaterial=null
       fontObject=null
       letterSpacing=0
       monospacedWidth=1
       size=9
       skewAmount=0
       typefaceFontName=Regular
       outlineSettings {
        bApplyOutlineToDropShadows=false
        bMiteredCorners=false
        bSeparateFillAlpha=false
        outlineMaterial=null
        outlineSize=0
        outlineColor=<rgba 0 0 0 1>
       }
      }
      highlightColor {colorUseRule=UseColor_ColorTable, specifiedColor=<rgba 0 0 0 0>}
      highlightShape {
       bIsDynamicallyLoaded=false
       drawAs=Image
       imageType=NoImage
       mirroring=NoMirror
       resourceName=None
       resourceObject=null
       tiling=NoTile
       imageSize=<0 0>
       margin=<ltrb 0 0 0 0>
       outlineSettings {
        bUseBrushTransparency=false
        roundingType=HalfHeightRadius
        width=0
        color {colorUseRule=UseColor_Specified, specifiedColor=<rgba 0 0 0 0>}
        cornerRadii=<0 0 0 0>
       }
       tintColor {colorUseRule=UseColor_ColorTable, specifiedColor=<rgba 0 0 0 0>}
       uVRegion {bIsValid=false, max=<0 0>, min=<0 0>}
      }
      selectedBackgroundColor {colorUseRule=UseColor_ColorTable, specifiedColor=<rgba 0 0 0 0>}
      shadowColorAndOpacity=<rgba 0 0 0 1>
      shadowOffset=<0 0>
      strikeBrush {
       bIsDynamicallyLoaded=false
       drawAs=Image
       imageType=NoImage
       mirroring=NoMirror
       resourceName=None
       resourceObject=null
       tiling=NoTile
       imageSize=<32 32>
       margin=<ltrb 0 0 0 0>
       outlineSettings {
        bUseBrushTransparency=false
        roundingType=HalfHeightRadius
        width=0
        color {colorUseRule=UseColor_Specified, specifiedColor=<rgba 0 0 0 0>}
        cornerRadii=<0 0 0 0>
       }
       tintColor {colorUseRule=UseColor_Specified, specifiedColor=<rgba 1 1 1 1>}
       uVRegion {bIsValid=false, max=<0 0>, min=<0 0>}
      }
      underlineBrush {
       bIsDynamicallyLoaded=false
       drawAs=Image
       imageType=NoImage
       mirroring=NoMirror
       resourceName=None
       resourceObject=null
       tiling=NoTile
       imageSize=<32 32>
       margin=<ltrb 0 0 0 0>
       outlineSettings {
        bUseBrushTransparency=false
        roundingType=HalfHeightRadius
        width=0
        color {colorUseRule=UseColor_Specified, specifiedColor=<rgba 0 0 0 0>}
        cornerRadii=<0 0 0 0>
       }
       tintColor {colorUseRule=UseColor_Specified, specifiedColor=<rgba 1 1 1 1>}
       uVRegion {bIsValid=false, max=<0 0>, min=<0 0>}
      }
     }
     vScrollBarPadding=<ltrb 0 0 0 0>
    }}
   slot {class=HorizontalBoxSlot, properties {size {sizeRule=Fill, value=1}}}
  }
  RenameAnimMontageHorizontalBox {
   widgetClassPath=&HorizontalBox
   bIsVariable=true
   parent=RenamingVerticalBox
   widget=RenameAnimMontageHorizontalBox
   slot {class=VerticalBoxSlot}
  }
  HorizontalBox_6 {
   widgetClassPath=&HorizontalBox
   parent=RenameAnimMontageHorizontalBox
   widget=HorizontalBox_6
   properties {bOverride_Cursor=true}
   slot {class=HorizontalBoxSlot, properties {padding=<ltrb 0 0 10 0>, size {sizeRule=Fill, value=1}}}
  }
  TextBlock_5 {
   widgetClassPath=&TextBlock
   parent=HorizontalBox_6
   widget=TextBlock_5
   properties {
    text=From
    colorAndOpacity {colorUseRule=UseColor_Specified, specifiedColor=<rgba 0.1274 0.1274 1 1>}
    font {
     bForceMonospaced=false
     bMaterialIsStencil=false
     fontMaterial=null
     fontObject=&Roboto
     letterSpacing=0
     monospacedWidth=1
     size=9
     skewAmount=0
     typefaceFontName=Bold
     outlineSettings {
      bApplyOutlineToDropShadows=false
      bMiteredCorners=false
      bSeparateFillAlpha=false
      outlineMaterial=null
      outlineSize=0
      outlineColor=<rgba 0 0 0 1>
     }
    }
   }
   slot {class=HorizontalBoxSlot, properties {verticalAlignment=VAlign_Center, padding=<ltrb 0 0 5 0>}}
  }
  ReplaceAMFromEditableText {
   widgetClassPath=&EditorUtilityEditableTextBox
   bIsVariable=true
   parent=HorizontalBox_6
   widget=ReplaceAMFromEditableText
   properties {widgetStyle {
     backgroundColor {colorUseRule=UseColor_Specified, specifiedColor=<rgba 1 1 1 1>}
     backgroundImageFocused {
      bIsDynamicallyLoaded=false
      drawAs=RoundedBox
      imageType=NoImage
      mirroring=NoMirror
      resourceName=None
      resourceObject=null
      tiling=NoTile
      imageSize=<0 0>
      margin=<ltrb 0 0 0 0>
      outlineSettings {
       bUseBrushTransparency=false
       roundingType=FixedRadius
       width=1
       color {colorUseRule=UseColor_ColorTable, specifiedColor=<rgba 0 0 0 0>}
       cornerRadii=<4 4 4 4>
      }
      tintColor {colorUseRule=UseColor_Specified, specifiedColor=<rgba 0.0048 0.0048 0.0048 1>}
      uVRegion {bIsValid=false, max=<0 0>, min=<0 0>}
     }
     backgroundImageHovered {
      bIsDynamicallyLoaded=false
      drawAs=RoundedBox
      imageType=NoImage
      mirroring=NoMirror
      resourceName=None
      resourceObject=null
      tiling=NoTile
      imageSize=<0 0>
      margin=<ltrb 0 0 0 0>
      outlineSettings {
       bUseBrushTransparency=false
       roundingType=FixedRadius
       width=1
       color {colorUseRule=UseColor_ColorTable, specifiedColor=<rgba 0 0 0 0>}
       cornerRadii=<4 4 4 4>
      }
      tintColor {colorUseRule=UseColor_Specified, specifiedColor=<rgba 0.0048 0.0048 0.0048 1>}
      uVRegion {bIsValid=false, max=<0 0>, min=<0 0>}
     }
     backgroundImageNormal {
      bIsDynamicallyLoaded=false
      drawAs=RoundedBox
      imageType=NoImage
      mirroring=NoMirror
      resourceName=None
      resourceObject=null
      tiling=NoTile
      imageSize=<0 0>
      margin=<ltrb 0 0 0 0>
      outlineSettings {
       bUseBrushTransparency=false
       roundingType=FixedRadius
       width=1
       color {colorUseRule=UseColor_ColorTable, specifiedColor=<rgba 0 0 0 0>}
       cornerRadii=<4 4 4 4>
      }
      tintColor {colorUseRule=UseColor_Specified, specifiedColor=<rgba 0.0048 0.0048 0.0048 1>}
      uVRegion {bIsValid=false, max=<0 0>, min=<0 0>}
     }
     backgroundImageReadOnly {
      bIsDynamicallyLoaded=false
      drawAs=RoundedBox
      imageType=NoImage
      mirroring=NoMirror
      resourceName=None
      resourceObject=null
      tiling=NoTile
      imageSize=<0 0>
      margin=<ltrb 0 0 0 0>
      outlineSettings {
       bUseBrushTransparency=false
       roundingType=FixedRadius
       width=1
       color {colorUseRule=UseColor_ColorTable, specifiedColor=<rgba 0 0 0 0>}
       cornerRadii=<4 4 4 4>
      }
      tintColor {colorUseRule=UseColor_Specified, specifiedColor=<rgba 0.0284 0.0284 0.0284 1>}
      uVRegion {bIsValid=false, max=<0 0>, min=<0 0>}
     }
     focusedForegroundColor {colorUseRule=UseColor_Specified, specifiedColor=<rgba 1 1 1 1>}
     foregroundColor {colorUseRule=UseColor_Specified, specifiedColor=<rgba 0.5271 0.5271 0.5271 1>}
     hScrollBarPadding=<ltrb 0 0 0 0>
     padding=<ltrb 8 3 8 4>
     readOnlyForegroundColor {colorUseRule=UseColor_Foreground, specifiedColor=<rgba 1 0 1 1>}
     scrollBarStyle {
      thickness=8
      draggedThumbImage {
       bIsDynamicallyLoaded=false
       drawAs=RoundedBox
       imageType=NoImage
       mirroring=NoMirror
       resourceName=None
       resourceObject=null
       tiling=NoTile
       imageSize=<0 0>
       margin=<ltrb 0 0 0 0>
       outlineSettings {
        bUseBrushTransparency=false
        roundingType=FixedRadius
        width=0
        color {colorUseRule=UseColor_Specified, specifiedColor=<rgba 0 0 0 0>}
        cornerRadii=<4 4 4 4>
       }
       tintColor {colorUseRule=UseColor_Specified, specifiedColor=<rgba 0.2159 0.2159 0.2159 1>}
       uVRegion {bIsValid=false, max=<0 0>, min=<0 0>}
      }
      horizontalBackgroundImage {
       bIsDynamicallyLoaded=false
       drawAs=NoDrawType
       imageType=NoImage
       mirroring=NoMirror
       resourceName=None
       resourceObject=null
       tiling=NoTile
       imageSize=<0 0>
       margin=<ltrb 0 0 0 0>
       outlineSettings {
        bUseBrushTransparency=false
        roundingType=HalfHeightRadius
        width=0
        color {colorUseRule=UseColor_Specified, specifiedColor=<rgba 0 0 0 0>}
        cornerRadii=<0 0 0 0>
       }
       tintColor {colorUseRule=UseColor_Specified, specifiedColor=<rgba 1 1 1 1>}
       uVRegion {bIsValid=false, max=<0 0>, min=<0 0>}
      }
      horizontalBottomSlotImage {
       bIsDynamicallyLoaded=false
       drawAs=NoDrawType
       imageType=NoImage
       mirroring=NoMirror
       resourceName=None
       resourceObject=null
       tiling=NoTile
       imageSize=<0 0>
       margin=<ltrb 0 0 0 0>
       outlineSettings {
        bUseBrushTransparency=false
        roundingType=HalfHeightRadius
        width=0
        color {colorUseRule=UseColor_Specified, specifiedColor=<rgba 0 0 0 0>}
        cornerRadii=<0 0 0 0>
       }
       tintColor {colorUseRule=UseColor_Specified, specifiedColor=<rgba 1 1 1 1>}
       uVRegion {bIsValid=false, max=<0 0>, min=<0 0>}
      }
      horizontalTopSlotImage {
       bIsDynamicallyLoaded=false
       drawAs=NoDrawType
       imageType=NoImage
       mirroring=NoMirror
       resourceName=None
       resourceObject=null
       tiling=NoTile
       imageSize=<0 0>
       margin=<ltrb 0 0 0 0>
       outlineSettings {
        bUseBrushTransparency=false
        roundingType=HalfHeightRadius
        width=0
        color {colorUseRule=UseColor_Specified, specifiedColor=<rgba 0 0 0 0>}
        cornerRadii=<0 0 0 0>
       }
       tintColor {colorUseRule=UseColor_Specified, specifiedColor=<rgba 1 1 1 1>}
       uVRegion {bIsValid=false, max=<0 0>, min=<0 0>}
      }
      hoveredThumbImage {
       bIsDynamicallyLoaded=false
       drawAs=RoundedBox
       imageType=NoImage
       mirroring=NoMirror
       resourceName=None
       resourceObject=null
       tiling=NoTile
       imageSize=<0 0>
       margin=<ltrb 0 0 0 0>
       outlineSettings {
        bUseBrushTransparency=false
        roundingType=FixedRadius
        width=0
        color {colorUseRule=UseColor_Specified, specifiedColor=<rgba 0 0 0 0>}
        cornerRadii=<4 4 4 4>
       }
       tintColor {colorUseRule=UseColor_Specified, specifiedColor=<rgba 0.2159 0.2159 0.2159 1>}
       uVRegion {bIsValid=false, max=<0 0>, min=<0 0>}
      }
      normalThumbImage {
       bIsDynamicallyLoaded=false
       drawAs=RoundedBox
       imageType=NoImage
       mirroring=NoMirror
       resourceName=None
       resourceObject=null
       tiling=NoTile
       imageSize=<0 0>
       margin=<ltrb 0 0 0 0>
       outlineSettings {
        bUseBrushTransparency=false
        roundingType=FixedRadius
        width=0
        color {colorUseRule=UseColor_Specified, specifiedColor=<rgba 0 0 0 0>}
        cornerRadii=<4 4 4 4>
       }
       tintColor {colorUseRule=UseColor_Specified, specifiedColor=<rgba 0.0953 0.0953 0.0953 1>}
       uVRegion {bIsValid=false, max=<0 0>, min=<0 0>}
      }
      verticalBackgroundImage {
       bIsDynamicallyLoaded=false
       drawAs=NoDrawType
       imageType=NoImage
       mirroring=NoMirror
       resourceName=None
       resourceObject=null
       tiling=NoTile
       imageSize=<0 0>
       margin=<ltrb 0 0 0 0>
       outlineSettings {
        bUseBrushTransparency=false
        roundingType=HalfHeightRadius
        width=0
        color {colorUseRule=UseColor_Specified, specifiedColor=<rgba 0 0 0 0>}
        cornerRadii=<0 0 0 0>
       }
       tintColor {colorUseRule=UseColor_Specified, specifiedColor=<rgba 1 1 1 1>}
       uVRegion {bIsValid=false, max=<0 0>, min=<0 0>}
      }
      verticalBottomSlotImage {
       bIsDynamicallyLoaded=false
       drawAs=NoDrawType
       imageType=NoImage
       mirroring=NoMirror
       resourceName=None
       resourceObject=null
       tiling=NoTile
       imageSize=<0 0>
       margin=<ltrb 0 0 0 0>
       outlineSettings {
        bUseBrushTransparency=false
        roundingType=HalfHeightRadius
        width=0
        color {colorUseRule=UseColor_Specified, specifiedColor=<rgba 0 0 0 0>}
        cornerRadii=<0 0 0 0>
       }
       tintColor {colorUseRule=UseColor_Specified, specifiedColor=<rgba 1 1 1 1>}
       uVRegion {bIsValid=false, max=<0 0>, min=<0 0>}
      }
      verticalTopSlotImage {
       bIsDynamicallyLoaded=false
       drawAs=NoDrawType
       imageType=NoImage
       mirroring=NoMirror
       resourceName=None
       resourceObject=null
       tiling=NoTile
       imageSize=<0 0>
       margin=<ltrb 0 0 0 0>
       outlineSettings {
        bUseBrushTransparency=false
        roundingType=HalfHeightRadius
        width=0
        color {colorUseRule=UseColor_Specified, specifiedColor=<rgba 0 0 0 0>}
        cornerRadii=<0 0 0 0>
       }
       tintColor {colorUseRule=UseColor_Specified, specifiedColor=<rgba 1 1 1 1>}
       uVRegion {bIsValid=false, max=<0 0>, min=<0 0>}
      }
     }
     textStyle {
      overflowPolicy=Clip
      transformPolicy=None
      colorAndOpacity {colorUseRule=UseColor_Foreground, specifiedColor=<rgba 1 0 1 1>}
      font {
       bForceMonospaced=false
       bMaterialIsStencil=false
       fontMaterial=null
       fontObject=null
       letterSpacing=0
       monospacedWidth=1
       size=9
       skewAmount=0
       typefaceFontName=Regular
       outlineSettings {
        bApplyOutlineToDropShadows=false
        bMiteredCorners=false
        bSeparateFillAlpha=false
        outlineMaterial=null
        outlineSize=0
        outlineColor=<rgba 0 0 0 1>
       }
      }
      highlightColor {colorUseRule=UseColor_ColorTable, specifiedColor=<rgba 0 0 0 0>}
      highlightShape {
       bIsDynamicallyLoaded=false
       drawAs=Image
       imageType=NoImage
       mirroring=NoMirror
       resourceName=None
       resourceObject=null
       tiling=NoTile
       imageSize=<0 0>
       margin=<ltrb 0 0 0 0>
       outlineSettings {
        bUseBrushTransparency=false
        roundingType=HalfHeightRadius
        width=0
        color {colorUseRule=UseColor_Specified, specifiedColor=<rgba 0 0 0 0>}
        cornerRadii=<0 0 0 0>
       }
       tintColor {colorUseRule=UseColor_ColorTable, specifiedColor=<rgba 0 0 0 0>}
       uVRegion {bIsValid=false, max=<0 0>, min=<0 0>}
      }
      selectedBackgroundColor {colorUseRule=UseColor_ColorTable, specifiedColor=<rgba 0 0 0 0>}
      shadowColorAndOpacity=<rgba 0 0 0 1>
      shadowOffset=<0 0>
      strikeBrush {
       bIsDynamicallyLoaded=false
       drawAs=Image
       imageType=NoImage
       mirroring=NoMirror
       resourceName=None
       resourceObject=null
       tiling=NoTile
       imageSize=<32 32>
       margin=<ltrb 0 0 0 0>
       outlineSettings {
        bUseBrushTransparency=false
        roundingType=HalfHeightRadius
        width=0
        color {colorUseRule=UseColor_Specified, specifiedColor=<rgba 0 0 0 0>}
        cornerRadii=<0 0 0 0>
       }
       tintColor {colorUseRule=UseColor_Specified, specifiedColor=<rgba 1 1 1 1>}
       uVRegion {bIsValid=false, max=<0 0>, min=<0 0>}
      }
      underlineBrush {
       bIsDynamicallyLoaded=false
       drawAs=Image
       imageType=NoImage
       mirroring=NoMirror
       resourceName=None
       resourceObject=null
       tiling=NoTile
       imageSize=<32 32>
       margin=<ltrb 0 0 0 0>
       outlineSettings {
        bUseBrushTransparency=false
        roundingType=HalfHeightRadius
        width=0
        color {colorUseRule=UseColor_Specified, specifiedColor=<rgba 0 0 0 0>}
        cornerRadii=<0 0 0 0>
       }
       tintColor {colorUseRule=UseColor_Specified, specifiedColor=<rgba 1 1 1 1>}
       uVRegion {bIsValid=false, max=<0 0>, min=<0 0>}
      }
     }
     vScrollBarPadding=<ltrb 0 0 0 0>
    }}
   slot {class=HorizontalBoxSlot, properties {size {sizeRule=Fill, value=1}}}
  }
  HorizontalBox_7 {
   widgetClassPath=&HorizontalBox
   parent=RenameAnimMontageHorizontalBox
   widget=HorizontalBox_7
   properties {bOverride_Cursor=true}
   slot {class=HorizontalBoxSlot, properties {padding=<ltrb 10 0 0 0>, size {sizeRule=Fill, value=1}}}
  }
  TextBlock_6 {
   widgetClassPath=&TextBlock
   parent=HorizontalBox_7
   widget=TextBlock_6
   properties {
    text=To
    colorAndOpacity {colorUseRule=UseColor_Specified, specifiedColor=<rgba 0.1274 0.1274 1 1>}
    font {
     bForceMonospaced=false
     bMaterialIsStencil=false
     fontMaterial=null
     fontObject=&Roboto
     letterSpacing=0
     monospacedWidth=1
     size=9
     skewAmount=0
     typefaceFontName=Bold
     outlineSettings {
      bApplyOutlineToDropShadows=false
      bMiteredCorners=false
      bSeparateFillAlpha=false
      outlineMaterial=null
      outlineSize=0
      outlineColor=<rgba 0 0 0 1>
     }
    }
   }
   slot {class=HorizontalBoxSlot, properties {verticalAlignment=VAlign_Center, padding=<ltrb 0 0 5 0>}}
  }
  ReplaceAMToEditableText {
   widgetClassPath=&EditorUtilityEditableTextBox
   bIsVariable=true
   parent=HorizontalBox_7
   widget=ReplaceAMToEditableText
   properties {widgetStyle {
     backgroundColor {colorUseRule=UseColor_Specified, specifiedColor=<rgba 1 1 1 1>}
     backgroundImageFocused {
      bIsDynamicallyLoaded=false
      drawAs=RoundedBox
      imageType=NoImage
      mirroring=NoMirror
      resourceName=None
      resourceObject=null
      tiling=NoTile
      imageSize=<0 0>
      margin=<ltrb 0 0 0 0>
      outlineSettings {
       bUseBrushTransparency=false
       roundingType=FixedRadius
       width=1
       color {colorUseRule=UseColor_ColorTable, specifiedColor=<rgba 0 0 0 0>}
       cornerRadii=<4 4 4 4>
      }
      tintColor {colorUseRule=UseColor_Specified, specifiedColor=<rgba 0.0048 0.0048 0.0048 1>}
      uVRegion {bIsValid=false, max=<0 0>, min=<0 0>}
     }
     backgroundImageHovered {
      bIsDynamicallyLoaded=false
      drawAs=RoundedBox
      imageType=NoImage
      mirroring=NoMirror
      resourceName=None
      resourceObject=null
      tiling=NoTile
      imageSize=<0 0>
      margin=<ltrb 0 0 0 0>
      outlineSettings {
       bUseBrushTransparency=false
       roundingType=FixedRadius
       width=1
       color {colorUseRule=UseColor_ColorTable, specifiedColor=<rgba 0 0 0 0>}
       cornerRadii=<4 4 4 4>
      }
      tintColor {colorUseRule=UseColor_Specified, specifiedColor=<rgba 0.0048 0.0048 0.0048 1>}
      uVRegion {bIsValid=false, max=<0 0>, min=<0 0>}
     }
     backgroundImageNormal {
      bIsDynamicallyLoaded=false
      drawAs=RoundedBox
      imageType=NoImage
      mirroring=NoMirror
      resourceName=None
      resourceObject=null
      tiling=NoTile
      imageSize=<0 0>
      margin=<ltrb 0 0 0 0>
      outlineSettings {
       bUseBrushTransparency=false
       roundingType=FixedRadius
       width=1
       color {colorUseRule=UseColor_ColorTable, specifiedColor=<rgba 0 0 0 0>}
       cornerRadii=<4 4 4 4>
      }
      tintColor {colorUseRule=UseColor_Specified, specifiedColor=<rgba 0.0048 0.0048 0.0048 1>}
      uVRegion {bIsValid=false, max=<0 0>, min=<0 0>}
     }
     backgroundImageReadOnly {
      bIsDynamicallyLoaded=false
      drawAs=RoundedBox
      imageType=NoImage
      mirroring=NoMirror
      resourceName=None
      resourceObject=null
      tiling=NoTile
      imageSize=<0 0>
      margin=<ltrb 0 0 0 0>
      outlineSettings {
       bUseBrushTransparency=false
       roundingType=FixedRadius
       width=1
       color {colorUseRule=UseColor_ColorTable, specifiedColor=<rgba 0 0 0 0>}
       cornerRadii=<4 4 4 4>
      }
      tintColor {colorUseRule=UseColor_Specified, specifiedColor=<rgba 0.0284 0.0284 0.0284 1>}
      uVRegion {bIsValid=false, max=<0 0>, min=<0 0>}
     }
     focusedForegroundColor {colorUseRule=UseColor_Specified, specifiedColor=<rgba 1 1 1 1>}
     foregroundColor {colorUseRule=UseColor_Specified, specifiedColor=<rgba 0.5271 0.5271 0.5271 1>}
     hScrollBarPadding=<ltrb 0 0 0 0>
     padding=<ltrb 8 3 8 4>
     readOnlyForegroundColor {colorUseRule=UseColor_Foreground, specifiedColor=<rgba 1 0 1 1>}
     scrollBarStyle {
      thickness=8
      draggedThumbImage {
       bIsDynamicallyLoaded=false
       drawAs=RoundedBox
       imageType=NoImage
       mirroring=NoMirror
       resourceName=None
       resourceObject=null
       tiling=NoTile
       imageSize=<0 0>
       margin=<ltrb 0 0 0 0>
       outlineSettings {
        bUseBrushTransparency=false
        roundingType=FixedRadius
        width=0
        color {colorUseRule=UseColor_Specified, specifiedColor=<rgba 0 0 0 0>}
        cornerRadii=<4 4 4 4>
       }
       tintColor {colorUseRule=UseColor_Specified, specifiedColor=<rgba 0.2159 0.2159 0.2159 1>}
       uVRegion {bIsValid=false, max=<0 0>, min=<0 0>}
      }
      horizontalBackgroundImage {
       bIsDynamicallyLoaded=false
       drawAs=NoDrawType
       imageType=NoImage
       mirroring=NoMirror
       resourceName=None
       resourceObject=null
       tiling=NoTile
       imageSize=<0 0>
       margin=<ltrb 0 0 0 0>
       outlineSettings {
        bUseBrushTransparency=false
        roundingType=HalfHeightRadius
        width=0
        color {colorUseRule=UseColor_Specified, specifiedColor=<rgba 0 0 0 0>}
        cornerRadii=<0 0 0 0>
       }
       tintColor {colorUseRule=UseColor_Specified, specifiedColor=<rgba 1 1 1 1>}
       uVRegion {bIsValid=false, max=<0 0>, min=<0 0>}
      }
      horizontalBottomSlotImage {
       bIsDynamicallyLoaded=false
       drawAs=NoDrawType
       imageType=NoImage
       mirroring=NoMirror
       resourceName=None
       resourceObject=null
       tiling=NoTile
       imageSize=<0 0>
       margin=<ltrb 0 0 0 0>
       outlineSettings {
        bUseBrushTransparency=false
        roundingType=HalfHeightRadius
        width=0
        color {colorUseRule=UseColor_Specified, specifiedColor=<rgba 0 0 0 0>}
        cornerRadii=<0 0 0 0>
       }
       tintColor {colorUseRule=UseColor_Specified, specifiedColor=<rgba 1 1 1 1>}
       uVRegion {bIsValid=false, max=<0 0>, min=<0 0>}
      }
      horizontalTopSlotImage {
       bIsDynamicallyLoaded=false
       drawAs=NoDrawType
       imageType=NoImage
       mirroring=NoMirror
       resourceName=None
       resourceObject=null
       tiling=NoTile
       imageSize=<0 0>
       margin=<ltrb 0 0 0 0>
       outlineSettings {
        bUseBrushTransparency=false
        roundingType=HalfHeightRadius
        width=0
        color {colorUseRule=UseColor_Specified, specifiedColor=<rgba 0 0 0 0>}
        cornerRadii=<0 0 0 0>
       }
       tintColor {colorUseRule=UseColor_Specified, specifiedColor=<rgba 1 1 1 1>}
       uVRegion {bIsValid=false, max=<0 0>, min=<0 0>}
      }
      hoveredThumbImage {
       bIsDynamicallyLoaded=false
       drawAs=RoundedBox
       imageType=NoImage
       mirroring=NoMirror
       resourceName=None
       resourceObject=null
       tiling=NoTile
       imageSize=<0 0>
       margin=<ltrb 0 0 0 0>
       outlineSettings {
        bUseBrushTransparency=false
        roundingType=FixedRadius
        width=0
        color {colorUseRule=UseColor_Specified, specifiedColor=<rgba 0 0 0 0>}
        cornerRadii=<4 4 4 4>
       }
       tintColor {colorUseRule=UseColor_Specified, specifiedColor=<rgba 0.2159 0.2159 0.2159 1>}
       uVRegion {bIsValid=false, max=<0 0>, min=<0 0>}
      }
      normalThumbImage {
       bIsDynamicallyLoaded=false
       drawAs=RoundedBox
       imageType=NoImage
       mirroring=NoMirror
       resourceName=None
       resourceObject=null
       tiling=NoTile
       imageSize=<0 0>
       margin=<ltrb 0 0 0 0>
       outlineSettings {
        bUseBrushTransparency=false
        roundingType=FixedRadius
        width=0
        color {colorUseRule=UseColor_Specified, specifiedColor=<rgba 0 0 0 0>}
        cornerRadii=<4 4 4 4>
       }
       tintColor {colorUseRule=UseColor_Specified, specifiedColor=<rgba 0.0953 0.0953 0.0953 1>}
       uVRegion {bIsValid=false, max=<0 0>, min=<0 0>}
      }
      verticalBackgroundImage {
       bIsDynamicallyLoaded=false
       drawAs=NoDrawType
       imageType=NoImage
       mirroring=NoMirror
       resourceName=None
       resourceObject=null
       tiling=NoTile
       imageSize=<0 0>
       margin=<ltrb 0 0 0 0>
       outlineSettings {
        bUseBrushTransparency=false
        roundingType=HalfHeightRadius
        width=0
        color {colorUseRule=UseColor_Specified, specifiedColor=<rgba 0 0 0 0>}
        cornerRadii=<0 0 0 0>
       }
       tintColor {colorUseRule=UseColor_Specified, specifiedColor=<rgba 1 1 1 1>}
       uVRegion {bIsValid=false, max=<0 0>, min=<0 0>}
      }
      verticalBottomSlotImage {
       bIsDynamicallyLoaded=false
       drawAs=NoDrawType
       imageType=NoImage
       mirroring=NoMirror
       resourceName=None
       resourceObject=null
       tiling=NoTile
       imageSize=<0 0>
       margin=<ltrb 0 0 0 0>
       outlineSettings {
        bUseBrushTransparency=false
        roundingType=HalfHeightRadius
        width=0
        color {colorUseRule=UseColor_Specified, specifiedColor=<rgba 0 0 0 0>}
        cornerRadii=<0 0 0 0>
       }
       tintColor {colorUseRule=UseColor_Specified, specifiedColor=<rgba 1 1 1 1>}
       uVRegion {bIsValid=false, max=<0 0>, min=<0 0>}
      }
      verticalTopSlotImage {
       bIsDynamicallyLoaded=false
       drawAs=NoDrawType
       imageType=NoImage
       mirroring=NoMirror
       resourceName=None
       resourceObject=null
       tiling=NoTile
       imageSize=<0 0>
       margin=<ltrb 0 0 0 0>
       outlineSettings {
        bUseBrushTransparency=false
        roundingType=HalfHeightRadius
        width=0
        color {colorUseRule=UseColor_Specified, specifiedColor=<rgba 0 0 0 0>}
        cornerRadii=<0 0 0 0>
       }
       tintColor {colorUseRule=UseColor_Specified, specifiedColor=<rgba 1 1 1 1>}
       uVRegion {bIsValid=false, max=<0 0>, min=<0 0>}
      }
     }
     textStyle {
      overflowPolicy=Clip
      transformPolicy=None
      colorAndOpacity {colorUseRule=UseColor_Foreground, specifiedColor=<rgba 1 0 1 1>}
      font {
       bForceMonospaced=false
       bMaterialIsStencil=false
       fontMaterial=null
       fontObject=null
       letterSpacing=0
       monospacedWidth=1
       size=9
       skewAmount=0
       typefaceFontName=Regular
       outlineSettings {
        bApplyOutlineToDropShadows=false
        bMiteredCorners=false
        bSeparateFillAlpha=false
        outlineMaterial=null
        outlineSize=0
        outlineColor=<rgba 0 0 0 1>
       }
      }
      highlightColor {colorUseRule=UseColor_ColorTable, specifiedColor=<rgba 0 0 0 0>}
      highlightShape {
       bIsDynamicallyLoaded=false
       drawAs=Image
       imageType=NoImage
       mirroring=NoMirror
       resourceName=None
       resourceObject=null
       tiling=NoTile
       imageSize=<0 0>
       margin=<ltrb 0 0 0 0>
       outlineSettings {
        bUseBrushTransparency=false
        roundingType=HalfHeightRadius
        width=0
        color {colorUseRule=UseColor_Specified, specifiedColor=<rgba 0 0 0 0>}
        cornerRadii=<0 0 0 0>
       }
       tintColor {colorUseRule=UseColor_ColorTable, specifiedColor=<rgba 0 0 0 0>}
       uVRegion {bIsValid=false, max=<0 0>, min=<0 0>}
      }
      selectedBackgroundColor {colorUseRule=UseColor_ColorTable, specifiedColor=<rgba 0 0 0 0>}
      shadowColorAndOpacity=<rgba 0 0 0 1>
      shadowOffset=<0 0>
      strikeBrush {
       bIsDynamicallyLoaded=false
       drawAs=Image
       imageType=NoImage
       mirroring=NoMirror
       resourceName=None
       resourceObject=null
       tiling=NoTile
       imageSize=<32 32>
       margin=<ltrb 0 0 0 0>
       outlineSettings {
        bUseBrushTransparency=false
        roundingType=HalfHeightRadius
        width=0
        color {colorUseRule=UseColor_Specified, specifiedColor=<rgba 0 0 0 0>}
        cornerRadii=<0 0 0 0>
       }
       tintColor {colorUseRule=UseColor_Specified, specifiedColor=<rgba 1 1 1 1>}
       uVRegion {bIsValid=false, max=<0 0>, min=<0 0>}
      }
      underlineBrush {
       bIsDynamicallyLoaded=false
       drawAs=Image
       imageType=NoImage
       mirroring=NoMirror
       resourceName=None
       resourceObject=null
       tiling=NoTile
       imageSize=<32 32>
       margin=<ltrb 0 0 0 0>
       outlineSettings {
        bUseBrushTransparency=false
        roundingType=HalfHeightRadius
        width=0
        color {colorUseRule=UseColor_Specified, specifiedColor=<rgba 0 0 0 0>}
        cornerRadii=<0 0 0 0>
       }
       tintColor {colorUseRule=UseColor_Specified, specifiedColor=<rgba 1 1 1 1>}
       uVRegion {bIsValid=false, max=<0 0>, min=<0 0>}
      }
     }
     vScrollBarPadding=<ltrb 0 0 0 0>
    }}
   slot {class=HorizontalBoxSlot, properties {size {sizeRule=Fill, value=1}}}
  }
  Border {
   widgetClassPath=&Border
   parent=VerticalBox_1
   widget=Border
   properties {brushColor=<rgba 0 0 0 1>, padding=<ltrb 0 2 0 0>}
   slot {class=VerticalBoxSlot, properties {padding=<ltrb 10 0 10 0>}}
  }
  CheckboxesIncludeSelectVerticalBox {
   widgetClassPath=&VerticalBox
   parent=VerticalBox_1
   widget=CheckboxesIncludeSelectVerticalBox
   properties {bOverride_Cursor=true}
   slot {class=VerticalBoxSlot, properties {padding=<ltrb 10 10 10 10>}}
  }
  HorizontalBox_3 {widgetClassPath=&HorizontalBox, parent=CheckboxesIncludeSelectVerticalBox, widget=HorizontalBox_3, slot {class=VerticalBoxSlot}}
  SelectAllCheckbox {
   widgetClassPath=&EditorUtilityCheckBox
   bIsVariable=true
   parent=HorizontalBox_3
   widget=SelectAllCheckbox
   properties {
    checkedState=Checked
    widgetStyle {
     checkBoxType=CheckBox
     backgroundHoveredImage {
      bIsDynamicallyLoaded=false
      drawAs=RoundedBox
      imageType=NoImage
      mirroring=NoMirror
      resourceName=None
      resourceObject=null
      tiling=NoTile
      imageSize=<18 18>
      margin=<ltrb 0 0 0 0>
      outlineSettings {
       bUseBrushTransparency=false
       roundingType=FixedRadius
       width=1
       color {colorUseRule=UseColor_ColorTable, specifiedColor=<rgba 0 0 0 0>}
       cornerRadii=<3 3 3 3>
      }
      tintColor {colorUseRule=UseColor_Specified, specifiedColor=<rgba 0.0048 0.0048 0.0048 1>}
      uVRegion {bIsValid=false, max=<0 0>, min=<0 0>}
     }
     backgroundImage {
      bIsDynamicallyLoaded=false
      drawAs=RoundedBox
      imageType=NoImage
      mirroring=NoMirror
      resourceName=None
      resourceObject=null
      tiling=NoTile
      imageSize=<18 18>
      margin=<ltrb 0 0 0 0>
      outlineSettings {
       bUseBrushTransparency=false
       roundingType=FixedRadius
       width=1
       color {colorUseRule=UseColor_ColorTable, specifiedColor=<rgba 0 0 0 0>}
       cornerRadii=<3 3 3 3>
      }
      tintColor {colorUseRule=UseColor_Specified, specifiedColor=<rgba 0.0048 0.0048 0.0048 1>}
      uVRegion {bIsValid=false, max=<0 0>, min=<0 0>}
     }
     backgroundPressedImage {
      bIsDynamicallyLoaded=false
      drawAs=RoundedBox
      imageType=NoImage
      mirroring=NoMirror
      resourceName=None
      resourceObject=null
      tiling=NoTile
      imageSize=<18 18>
      margin=<ltrb 0 0 0 0>
      outlineSettings {
       bUseBrushTransparency=false
       roundingType=FixedRadius
       width=1
       color {colorUseRule=UseColor_ColorTable, specifiedColor=<rgba 0 0 0 0>}
       cornerRadii=<3 3 3 3>
      }
      tintColor {colorUseRule=UseColor_Specified, specifiedColor=<rgba 0.0048 0.0048 0.0048 1>}
      uVRegion {bIsValid=false, max=<0 0>, min=<0 0>}
     }
     borderBackgroundColor {colorUseRule=UseColor_Specified, specifiedColor=<rgba 1 1 1 1>}
     checkedForeground {colorUseRule=UseColor_Specified, specifiedColor=<rgba 1 1 1 1>}
     checkedHoveredForeground {colorUseRule=UseColor_Specified, specifiedColor=<rgba 1 1 1 1>}
     checkedHoveredImage {
      bIsDynamicallyLoaded=false
      drawAs=Image
      imageType=Vector
      mirroring=NoMirror
      resourceName=../../../Engine/Content/Slate/Starship/CoreWidgets/CheckBox/check.svg
      resourceObject=null
      tiling=NoTile
      imageSize=<16 16>
      margin=<ltrb 0 0 0 0>
      outlineSettings {
       bUseBrushTransparency=false
       roundingType=HalfHeightRadius
       width=0
       color {colorUseRule=UseColor_Specified, specifiedColor=<rgba 0 0 0 0>}
       cornerRadii=<0 0 0 0>
      }
      tintColor {colorUseRule=UseColor_Specified, specifiedColor=<rgba 0.0044 0.2384 1 1>}
      uVRegion {bIsValid=false, max=<0 0>, min=<0 0>}
     }
     checkedImage {
      bIsDynamicallyLoaded=false
      drawAs=Image
      imageType=Vector
      mirroring=NoMirror
      resourceName=../../../Engine/Content/Slate/Starship/CoreWidgets/CheckBox/check.svg
      resourceObject=null
      tiling=NoTile
      imageSize=<16 16>
      margin=<ltrb 0 0 0 0>
      outlineSettings {
       bUseBrushTransparency=false
       roundingType=HalfHeightRadius
       width=0
       color {colorUseRule=UseColor_Specified, specifiedColor=<rgba 0 0 0 0>}
       cornerRadii=<0 0 0 0>
      }
      tintColor {colorUseRule=UseColor_Specified, specifiedColor=<rgba 0 0.162 0.7454 1>}
      uVRegion {bIsValid=false, max=<0 0>, min=<0 0>}
     }
     checkedPressedForeground {colorUseRule=UseColor_Specified, specifiedColor=<rgba 1 1 1 1>}
     checkedPressedImage {
      bIsDynamicallyLoaded=false
      drawAs=Image
      imageType=Vector
      mirroring=NoMirror
      resourceName=../../../Engine/Content/Slate/Starship/CoreWidgets/CheckBox/check.svg
      resourceObject=null
      tiling=NoTile
      imageSize=<16 16>
      margin=<ltrb 0 0 0 0>
      outlineSettings {
       bUseBrushTransparency=false
       roundingType=HalfHeightRadius
       width=0
       color {colorUseRule=UseColor_Specified, specifiedColor=<rgba 0 0 0 0>}
       cornerRadii=<0 0 0 0>
      }
      tintColor {colorUseRule=UseColor_Specified, specifiedColor=<rgba 0 0.0802 0.3515 1>}
      uVRegion {bIsValid=false, max=<0 0>, min=<0 0>}
     }
     checkedSlateSound {resourceObject=null}
     foregroundColor {colorUseRule=UseColor_Specified, specifiedColor=<rgba 1 1 1 1>}
     hoveredForeground {colorUseRule=UseColor_Specified, specifiedColor=<rgba 1 1 1 1>}
     hoveredSlateSound {resourceObject=null}
     padding=<ltrb 0 0 0 0>
     pressedForeground {colorUseRule=UseColor_Specified, specifiedColor=<rgba 1 1 1 1>}
     uncheckedHoveredImage {
      bIsDynamicallyLoaded=false
      drawAs=NoDrawType
      imageType=NoImage
      mirroring=NoMirror
      resourceName=None
      resourceObject=null
      tiling=NoTile
      imageSize=<0 0>
      margin=<ltrb 0 0 0 0>
      outlineSettings {
       bUseBrushTransparency=false
       roundingType=HalfHeightRadius
       width=0
       color {colorUseRule=UseColor_Specified, specifiedColor=<rgba 0 0 0 0>}
       cornerRadii=<0 0 0 0>
      }
      tintColor {colorUseRule=UseColor_Specified, specifiedColor=<rgba 1 1 1 1>}
      uVRegion {bIsValid=false, max=<0 0>, min=<0 0>}
     }
     uncheckedImage {
      bIsDynamicallyLoaded=false
      drawAs=NoDrawType
      imageType=NoImage
      mirroring=NoMirror
      resourceName=None
      resourceObject=null
      tiling=NoTile
      imageSize=<0 0>
      margin=<ltrb 0 0 0 0>
      outlineSettings {
       bUseBrushTransparency=false
       roundingType=HalfHeightRadius
       width=0
       color {colorUseRule=UseColor_Specified, specifiedColor=<rgba 0 0 0 0>}
       cornerRadii=<0 0 0 0>
      }
      tintColor {colorUseRule=UseColor_Specified, specifiedColor=<rgba 1 1 1 1>}
      uVRegion {bIsValid=false, max=<0 0>, min=<0 0>}
     }
     uncheckedPressedImage {
      bIsDynamicallyLoaded=false
      drawAs=NoDrawType
      imageType=NoImage
      mirroring=NoMirror
      resourceName=None
      resourceObject=null
      tiling=NoTile
      imageSize=<0 0>
      margin=<ltrb 0 0 0 0>
      outlineSettings {
       bUseBrushTransparency=false
       roundingType=HalfHeightRadius
       width=0
       color {colorUseRule=UseColor_Specified, specifiedColor=<rgba 0 0 0 0>}
       cornerRadii=<0 0 0 0>
      }
      tintColor {colorUseRule=UseColor_Specified, specifiedColor=<rgba 1 1 1 1>}
      uVRegion {bIsValid=false, max=<0 0>, min=<0 0>}
     }
     uncheckedSlateSound {resourceObject=null}
     undeterminedForeground {colorUseRule=UseColor_Specified, specifiedColor=<rgba 1 1 1 1>}
     undeterminedHoveredImage {
      bIsDynamicallyLoaded=false
      drawAs=Image
      imageType=Vector
      mirroring=NoMirror
      resourceName=../../../Engine/Content/Slate/Starship/CoreWidgets/CheckBox/indeterminate.svg
      resourceObject=null
      tiling=NoTile
      imageSize=<16 16>
      margin=<ltrb 0 0 0 0>
      outlineSettings {
       bUseBrushTransparency=false
       roundingType=HalfHeightRadius
       width=0
       color {colorUseRule=UseColor_Specified, specifiedColor=<rgba 0 0 0 0>}
       cornerRadii=<0 0 0 0>
      }
      tintColor {colorUseRule=UseColor_Specified, specifiedColor=<rgba 0.0044 0.2384 1 1>}
      uVRegion {bIsValid=false, max=<0 0>, min=<0 0>}
     }
     undeterminedImage {
      bIsDynamicallyLoaded=false
      drawAs=Image
      imageType=Vector
      mirroring=NoMirror
      resourceName=../../../Engine/Content/Slate/Starship/CoreWidgets/CheckBox/indeterminate.svg
      resourceObject=null
      tiling=NoTile
      imageSize=<16 16>
      margin=<ltrb 0 0 0 0>
      outlineSettings {
       bUseBrushTransparency=false
       roundingType=HalfHeightRadius
       width=0
       color {colorUseRule=UseColor_Specified, specifiedColor=<rgba 0 0 0 0>}
       cornerRadii=<0 0 0 0>
      }
      tintColor {colorUseRule=UseColor_Specified, specifiedColor=<rgba 0 0.162 0.7454 1>}
      uVRegion {bIsValid=false, max=<0 0>, min=<0 0>}
     }
     undeterminedPressedImage {
      bIsDynamicallyLoaded=false
      drawAs=Image
      imageType=Vector
      mirroring=NoMirror
      resourceName=../../../Engine/Content/Slate/Starship/CoreWidgets/CheckBox/indeterminate.svg
      resourceObject=null
      tiling=NoTile
      imageSize=<16 16>
      margin=<ltrb 0 0 0 0>
      outlineSettings {
       bUseBrushTransparency=false
       roundingType=HalfHeightRadius
       width=0
       color {colorUseRule=UseColor_Specified, specifiedColor=<rgba 0 0 0 0>}
       cornerRadii=<0 0 0 0>
      }
      tintColor {colorUseRule=UseColor_Specified, specifiedColor=<rgba 0 0.0802 0.3515 1>}
      uVRegion {bIsValid=false, max=<0 0>, min=<0 0>}
     }
    }
   }
   slot {class=HorizontalBoxSlot, properties {padding=<ltrb 0 0 5 0>}}
  }
  TextBlock_184 {
   widgetClassPath=&TextBlock
   parent=HorizontalBox_3
   widget=TextBlock_184
   properties {
    text=Select All
    font {
     bForceMonospaced=false
     bMaterialIsStencil=false
     fontMaterial=null
     fontObject=&Roboto
     letterSpacing=0
     monospacedWidth=1
     size=7.5
     skewAmount=0
     typefaceFontName=Bold
     outlineSettings {
      bApplyOutlineToDropShadows=false
      bMiteredCorners=false
      bSeparateFillAlpha=false
      outlineMaterial=null
      outlineSize=0
      outlineColor=<rgba 0 0 0 1>
     }
    }
   }
   slot {class=HorizontalBoxSlot, properties {verticalAlignment=VAlign_Center, padding=<ltrb 0 0 15 0>}}
  }
  IncludeLevelSequencesCheckbox {
   widgetClassPath=&EditorUtilityCheckBox
   bIsVariable=true
   parent=HorizontalBox_3
   widget=IncludeLevelSequencesCheckbox
   properties {
    checkedState=Checked
    widgetStyle {
     checkBoxType=CheckBox
     backgroundHoveredImage {
      bIsDynamicallyLoaded=false
      drawAs=RoundedBox
      imageType=NoImage
      mirroring=NoMirror
      resourceName=None
      resourceObject=null
      tiling=NoTile
      imageSize=<18 18>
      margin=<ltrb 0 0 0 0>
      outlineSettings {
       bUseBrushTransparency=false
       roundingType=FixedRadius
       width=1
       color {colorUseRule=UseColor_ColorTable, specifiedColor=<rgba 0 0 0 0>}
       cornerRadii=<3 3 3 3>
      }
      tintColor {colorUseRule=UseColor_Specified, specifiedColor=<rgba 0.0048 0.0048 0.0048 1>}
      uVRegion {bIsValid=false, max=<0 0>, min=<0 0>}
     }
     backgroundImage {
      bIsDynamicallyLoaded=false
      drawAs=RoundedBox
      imageType=NoImage
      mirroring=NoMirror
      resourceName=None
      resourceObject=null
      tiling=NoTile
      imageSize=<18 18>
      margin=<ltrb 0 0 0 0>
      outlineSettings {
       bUseBrushTransparency=false
       roundingType=FixedRadius
       width=1
       color {colorUseRule=UseColor_ColorTable, specifiedColor=<rgba 0 0 0 0>}
       cornerRadii=<3 3 3 3>
      }
      tintColor {colorUseRule=UseColor_Specified, specifiedColor=<rgba 0.0048 0.0048 0.0048 1>}
      uVRegion {bIsValid=false, max=<0 0>, min=<0 0>}
     }
     backgroundPressedImage {
      bIsDynamicallyLoaded=false
      drawAs=RoundedBox
      imageType=NoImage
      mirroring=NoMirror
      resourceName=None
      resourceObject=null
      tiling=NoTile
      imageSize=<18 18>
      margin=<ltrb 0 0 0 0>
      outlineSettings {
       bUseBrushTransparency=false
       roundingType=FixedRadius
       width=1
       color {colorUseRule=UseColor_ColorTable, specifiedColor=<rgba 0 0 0 0>}
       cornerRadii=<3 3 3 3>
      }
      tintColor {colorUseRule=UseColor_Specified, specifiedColor=<rgba 0.0048 0.0048 0.0048 1>}
      uVRegion {bIsValid=false, max=<0 0>, min=<0 0>}
     }
     borderBackgroundColor {colorUseRule=UseColor_Specified, specifiedColor=<rgba 1 1 1 1>}
     checkedForeground {colorUseRule=UseColor_Specified, specifiedColor=<rgba 1 1 1 1>}
     checkedHoveredForeground {colorUseRule=UseColor_Specified, specifiedColor=<rgba 1 1 1 1>}
     checkedHoveredImage {
      bIsDynamicallyLoaded=false
      drawAs=Image
      imageType=Vector
      mirroring=NoMirror
      resourceName=../../../Engine/Content/Slate/Starship/CoreWidgets/CheckBox/check.svg
      resourceObject=null
      tiling=NoTile
      imageSize=<16 16>
      margin=<ltrb 0 0 0 0>
      outlineSettings {
       bUseBrushTransparency=false
       roundingType=HalfHeightRadius
       width=0
       color {colorUseRule=UseColor_Specified, specifiedColor=<rgba 0 0 0 0>}
       cornerRadii=<0 0 0 0>
      }
      tintColor {colorUseRule=UseColor_Specified, specifiedColor=<rgba 0.0044 0.2384 1 1>}
      uVRegion {bIsValid=false, max=<0 0>, min=<0 0>}
     }
     checkedImage {
      bIsDynamicallyLoaded=false
      drawAs=Image
      imageType=Vector
      mirroring=NoMirror
      resourceName=../../../Engine/Content/Slate/Starship/CoreWidgets/CheckBox/check.svg
      resourceObject=null
      tiling=NoTile
      imageSize=<16 16>
      margin=<ltrb 0 0 0 0>
      outlineSettings {
       bUseBrushTransparency=false
       roundingType=HalfHeightRadius
       width=0
       color {colorUseRule=UseColor_Specified, specifiedColor=<rgba 0 0 0 0>}
       cornerRadii=<0 0 0 0>
      }
      tintColor {colorUseRule=UseColor_Specified, specifiedColor=<rgba 0 0.162 0.7454 1>}
      uVRegion {bIsValid=false, max=<0 0>, min=<0 0>}
     }
     checkedPressedForeground {colorUseRule=UseColor_Specified, specifiedColor=<rgba 1 1 1 1>}
     checkedPressedImage {
      bIsDynamicallyLoaded=false
      drawAs=Image
      imageType=Vector
      mirroring=NoMirror
      resourceName=../../../Engine/Content/Slate/Starship/CoreWidgets/CheckBox/check.svg
      resourceObject=null
      tiling=NoTile
      imageSize=<16 16>
      margin=<ltrb 0 0 0 0>
      outlineSettings {
       bUseBrushTransparency=false
       roundingType=HalfHeightRadius
       width=0
       color {colorUseRule=UseColor_Specified, specifiedColor=<rgba 0 0 0 0>}
       cornerRadii=<0 0 0 0>
      }
      tintColor {colorUseRule=UseColor_Specified, specifiedColor=<rgba 0 0.0802 0.3515 1>}
      uVRegion {bIsValid=false, max=<0 0>, min=<0 0>}
     }
     checkedSlateSound {resourceObject=null}
     foregroundColor {colorUseRule=UseColor_Specified, specifiedColor=<rgba 1 1 1 1>}
     hoveredForeground {colorUseRule=UseColor_Specified, specifiedColor=<rgba 1 1 1 1>}
     hoveredSlateSound {resourceObject=null}
     padding=<ltrb 0 0 0 0>
     pressedForeground {colorUseRule=UseColor_Specified, specifiedColor=<rgba 1 1 1 1>}
     uncheckedHoveredImage {
      bIsDynamicallyLoaded=false
      drawAs=NoDrawType
      imageType=NoImage
      mirroring=NoMirror
      resourceName=None
      resourceObject=null
      tiling=NoTile
      imageSize=<0 0>
      margin=<ltrb 0 0 0 0>
      outlineSettings {
       bUseBrushTransparency=false
       roundingType=HalfHeightRadius
       width=0
       color {colorUseRule=UseColor_Specified, specifiedColor=<rgba 0 0 0 0>}
       cornerRadii=<0 0 0 0>
      }
      tintColor {colorUseRule=UseColor_Specified, specifiedColor=<rgba 1 1 1 1>}
      uVRegion {bIsValid=false, max=<0 0>, min=<0 0>}
     }
     uncheckedImage {
      bIsDynamicallyLoaded=false
      drawAs=NoDrawType
      imageType=NoImage
      mirroring=NoMirror
      resourceName=None
      resourceObject=null
      tiling=NoTile
      imageSize=<0 0>
      margin=<ltrb 0 0 0 0>
      outlineSettings {
       bUseBrushTransparency=false
       roundingType=HalfHeightRadius
       width=0
       color {colorUseRule=UseColor_Specified, specifiedColor=<rgba 0 0 0 0>}
       cornerRadii=<0 0 0 0>
      }
      tintColor {colorUseRule=UseColor_Specified, specifiedColor=<rgba 1 1 1 1>}
      uVRegion {bIsValid=false, max=<0 0>, min=<0 0>}
     }
     uncheckedPressedImage {
      bIsDynamicallyLoaded=false
      drawAs=NoDrawType
      imageType=NoImage
      mirroring=NoMirror
      resourceName=None
      resourceObject=null
      tiling=NoTile
      imageSize=<0 0>
      margin=<ltrb 0 0 0 0>
      outlineSettings {
       bUseBrushTransparency=false
       roundingType=HalfHeightRadius
       width=0
       color {colorUseRule=UseColor_Specified, specifiedColor=<rgba 0 0 0 0>}
       cornerRadii=<0 0 0 0>
      }
      tintColor {colorUseRule=UseColor_Specified, specifiedColor=<rgba 1 1 1 1>}
      uVRegion {bIsValid=false, max=<0 0>, min=<0 0>}
     }
     uncheckedSlateSound {resourceObject=null}
     undeterminedForeground {colorUseRule=UseColor_Specified, specifiedColor=<rgba 1 1 1 1>}
     undeterminedHoveredImage {
      bIsDynamicallyLoaded=false
      drawAs=Image
      imageType=Vector
      mirroring=NoMirror
      resourceName=../../../Engine/Content/Slate/Starship/CoreWidgets/CheckBox/indeterminate.svg
      resourceObject=null
      tiling=NoTile
      imageSize=<16 16>
      margin=<ltrb 0 0 0 0>
      outlineSettings {
       bUseBrushTransparency=false
       roundingType=HalfHeightRadius
       width=0
       color {colorUseRule=UseColor_Specified, specifiedColor=<rgba 0 0 0 0>}
       cornerRadii=<0 0 0 0>
      }
      tintColor {colorUseRule=UseColor_Specified, specifiedColor=<rgba 0.0044 0.2384 1 1>}
      uVRegion {bIsValid=false, max=<0 0>, min=<0 0>}
     }
     undeterminedImage {
      bIsDynamicallyLoaded=false
      drawAs=Image
      imageType=Vector
      mirroring=NoMirror
      resourceName=../../../Engine/Content/Slate/Starship/CoreWidgets/CheckBox/indeterminate.svg
      resourceObject=null
      tiling=NoTile
      imageSize=<16 16>
      margin=<ltrb 0 0 0 0>
      outlineSettings {
       bUseBrushTransparency=false
       roundingType=HalfHeightRadius
       width=0
       color {colorUseRule=UseColor_Specified, specifiedColor=<rgba 0 0 0 0>}
       cornerRadii=<0 0 0 0>
      }
      tintColor {colorUseRule=UseColor_Specified, specifiedColor=<rgba 0 0.162 0.7454 1>}
      uVRegion {bIsValid=false, max=<0 0>, min=<0 0>}
     }
     undeterminedPressedImage {
      bIsDynamicallyLoaded=false
      drawAs=Image
      imageType=Vector
      mirroring=NoMirror
      resourceName=../../../Engine/Content/Slate/Starship/CoreWidgets/CheckBox/indeterminate.svg
      resourceObject=null
      tiling=NoTile
      imageSize=<16 16>
      margin=<ltrb 0 0 0 0>
      outlineSettings {
       bUseBrushTransparency=false
       roundingType=HalfHeightRadius
       width=0
       color {colorUseRule=UseColor_Specified, specifiedColor=<rgba 0 0 0 0>}
       cornerRadii=<0 0 0 0>
      }
      tintColor {colorUseRule=UseColor_Specified, specifiedColor=<rgba 0 0.0802 0.3515 1>}
      uVRegion {bIsValid=false, max=<0 0>, min=<0 0>}
     }
    }
   }
   slot {class=HorizontalBoxSlot, properties {padding=<ltrb 0 0 5 0>}}
  }
  TextBlock_10 {
   widgetClassPath=&TextBlock
   parent=HorizontalBox_3
   widget=TextBlock_10
   properties {
    text=Include LevelSequences
    colorAndOpacity {colorUseRule=UseColor_Specified, specifiedColor=<rgba 0.5776 0.0802 0.0802 1>}
    font {
     bForceMonospaced=false
     bMaterialIsStencil=false
     fontMaterial=null
     fontObject=&Roboto
     letterSpacing=0
     monospacedWidth=1
     size=7.5
     skewAmount=0
     typefaceFontName=Bold
     outlineSettings {
      bApplyOutlineToDropShadows=false
      bMiteredCorners=false
      bSeparateFillAlpha=false
      outlineMaterial=null
      outlineSize=0
      outlineColor=<rgba 0 0 0 1>
     }
    }
   }
   slot {class=HorizontalBoxSlot, properties {verticalAlignment=VAlign_Center, padding=<ltrb 0 0 10 0>}}
  }
  IncludeAnimSequencesCheckbox {
   widgetClassPath=&EditorUtilityCheckBox
   bIsVariable=true
   parent=HorizontalBox_3
   widget=IncludeAnimSequencesCheckbox
   properties {
    checkedState=Checked
    widgetStyle {
     checkBoxType=CheckBox
     backgroundHoveredImage {
      bIsDynamicallyLoaded=false
      drawAs=RoundedBox
      imageType=NoImage
      mirroring=NoMirror
      resourceName=None
      resourceObject=null
      tiling=NoTile
      imageSize=<18 18>
      margin=<ltrb 0 0 0 0>
      outlineSettings {
       bUseBrushTransparency=false
       roundingType=FixedRadius
       width=1
       color {colorUseRule=UseColor_ColorTable, specifiedColor=<rgba 0 0 0 0>}
       cornerRadii=<3 3 3 3>
      }
      tintColor {colorUseRule=UseColor_Specified, specifiedColor=<rgba 0.0048 0.0048 0.0048 1>}
      uVRegion {bIsValid=false, max=<0 0>, min=<0 0>}
     }
     backgroundImage {
      bIsDynamicallyLoaded=false
      drawAs=RoundedBox
      imageType=NoImage
      mirroring=NoMirror
      resourceName=None
      resourceObject=null
      tiling=NoTile
      imageSize=<18 18>
      margin=<ltrb 0 0 0 0>
      outlineSettings {
       bUseBrushTransparency=false
       roundingType=FixedRadius
       width=1
       color {colorUseRule=UseColor_ColorTable, specifiedColor=<rgba 0 0 0 0>}
       cornerRadii=<3 3 3 3>
      }
      tintColor {colorUseRule=UseColor_Specified, specifiedColor=<rgba 0.0048 0.0048 0.0048 1>}
      uVRegion {bIsValid=false, max=<0 0>, min=<0 0>}
     }
     backgroundPressedImage {
      bIsDynamicallyLoaded=false
      drawAs=RoundedBox
      imageType=NoImage
      mirroring=NoMirror
      resourceName=None
      resourceObject=null
      tiling=NoTile
      imageSize=<18 18>
      margin=<ltrb 0 0 0 0>
      outlineSettings {
       bUseBrushTransparency=false
       roundingType=FixedRadius
       width=1
       color {colorUseRule=UseColor_ColorTable, specifiedColor=<rgba 0 0 0 0>}
       cornerRadii=<3 3 3 3>
      }
      tintColor {colorUseRule=UseColor_Specified, specifiedColor=<rgba 0.0048 0.0048 0.0048 1>}
      uVRegion {bIsValid=false, max=<0 0>, min=<0 0>}
     }
     borderBackgroundColor {colorUseRule=UseColor_Specified, specifiedColor=<rgba 1 1 1 1>}
     checkedForeground {colorUseRule=UseColor_Specified, specifiedColor=<rgba 1 1 1 1>}
     checkedHoveredForeground {colorUseRule=UseColor_Specified, specifiedColor=<rgba 1 1 1 1>}
     checkedHoveredImage {
      bIsDynamicallyLoaded=false
      drawAs=Image
      imageType=Vector
      mirroring=NoMirror
      resourceName=../../../Engine/Content/Slate/Starship/CoreWidgets/CheckBox/check.svg
      resourceObject=null
      tiling=NoTile
      imageSize=<16 16>
      margin=<ltrb 0 0 0 0>
      outlineSettings {
       bUseBrushTransparency=false
       roundingType=HalfHeightRadius
       width=0
       color {colorUseRule=UseColor_Specified, specifiedColor=<rgba 0 0 0 0>}
       cornerRadii=<0 0 0 0>
      }
      tintColor {colorUseRule=UseColor_Specified, specifiedColor=<rgba 0.0044 0.2384 1 1>}
      uVRegion {bIsValid=false, max=<0 0>, min=<0 0>}
     }
     checkedImage {
      bIsDynamicallyLoaded=false
      drawAs=Image
      imageType=Vector
      mirroring=NoMirror
      resourceName=../../../Engine/Content/Slate/Starship/CoreWidgets/CheckBox/check.svg
      resourceObject=null
      tiling=NoTile
      imageSize=<16 16>
      margin=<ltrb 0 0 0 0>
      outlineSettings {
       bUseBrushTransparency=false
       roundingType=HalfHeightRadius
       width=0
       color {colorUseRule=UseColor_Specified, specifiedColor=<rgba 0 0 0 0>}
       cornerRadii=<0 0 0 0>
      }
      tintColor {colorUseRule=UseColor_Specified, specifiedColor=<rgba 0 0.162 0.7454 1>}
      uVRegion {bIsValid=false, max=<0 0>, min=<0 0>}
     }
     checkedPressedForeground {colorUseRule=UseColor_Specified, specifiedColor=<rgba 1 1 1 1>}
     checkedPressedImage {
      bIsDynamicallyLoaded=false
      drawAs=Image
      imageType=Vector
      mirroring=NoMirror
      resourceName=../../../Engine/Content/Slate/Starship/CoreWidgets/CheckBox/check.svg
      resourceObject=null
      tiling=NoTile
      imageSize=<16 16>
      margin=<ltrb 0 0 0 0>
      outlineSettings {
       bUseBrushTransparency=false
       roundingType=HalfHeightRadius
       width=0
       color {colorUseRule=UseColor_Specified, specifiedColor=<rgba 0 0 0 0>}
       cornerRadii=<0 0 0 0>
      }
      tintColor {colorUseRule=UseColor_Specified, specifiedColor=<rgba 0 0.0802 0.3515 1>}
      uVRegion {bIsValid=false, max=<0 0>, min=<0 0>}
     }
     checkedSlateSound {resourceObject=null}
     foregroundColor {colorUseRule=UseColor_Specified, specifiedColor=<rgba 1 1 1 1>}
     hoveredForeground {colorUseRule=UseColor_Specified, specifiedColor=<rgba 1 1 1 1>}
     hoveredSlateSound {resourceObject=null}
     padding=<ltrb 0 0 0 0>
     pressedForeground {colorUseRule=UseColor_Specified, specifiedColor=<rgba 1 1 1 1>}
     uncheckedHoveredImage {
      bIsDynamicallyLoaded=false
      drawAs=NoDrawType
      imageType=NoImage
      mirroring=NoMirror
      resourceName=None
      resourceObject=null
      tiling=NoTile
      imageSize=<0 0>
      margin=<ltrb 0 0 0 0>
      outlineSettings {
       bUseBrushTransparency=false
       roundingType=HalfHeightRadius
       width=0
       color {colorUseRule=UseColor_Specified, specifiedColor=<rgba 0 0 0 0>}
       cornerRadii=<0 0 0 0>
      }
      tintColor {colorUseRule=UseColor_Specified, specifiedColor=<rgba 1 1 1 1>}
      uVRegion {bIsValid=false, max=<0 0>, min=<0 0>}
     }
     uncheckedImage {
      bIsDynamicallyLoaded=false
      drawAs=NoDrawType
      imageType=NoImage
      mirroring=NoMirror
      resourceName=None
      resourceObject=null
      tiling=NoTile
      imageSize=<0 0>
      margin=<ltrb 0 0 0 0>
      outlineSettings {
       bUseBrushTransparency=false
       roundingType=HalfHeightRadius
       width=0
       color {colorUseRule=UseColor_Specified, specifiedColor=<rgba 0 0 0 0>}
       cornerRadii=<0 0 0 0>
      }
      tintColor {colorUseRule=UseColor_Specified, specifiedColor=<rgba 1 1 1 1>}
      uVRegion {bIsValid=false, max=<0 0>, min=<0 0>}
     }
     uncheckedPressedImage {
      bIsDynamicallyLoaded=false
      drawAs=NoDrawType
      imageType=NoImage
      mirroring=NoMirror
      resourceName=None
      resourceObject=null
      tiling=NoTile
      imageSize=<0 0>
      margin=<ltrb 0 0 0 0>
      outlineSettings {
       bUseBrushTransparency=false
       roundingType=HalfHeightRadius
       width=0
       color {colorUseRule=UseColor_Specified, specifiedColor=<rgba 0 0 0 0>}
       cornerRadii=<0 0 0 0>
      }
      tintColor {colorUseRule=UseColor_Specified, specifiedColor=<rgba 1 1 1 1>}
      uVRegion {bIsValid=false, max=<0 0>, min=<0 0>}
     }
     uncheckedSlateSound {resourceObject=null}
     undeterminedForeground {colorUseRule=UseColor_Specified, specifiedColor=<rgba 1 1 1 1>}
     undeterminedHoveredImage {
      bIsDynamicallyLoaded=false
      drawAs=Image
      imageType=Vector
      mirroring=NoMirror
      resourceName=../../../Engine/Content/Slate/Starship/CoreWidgets/CheckBox/indeterminate.svg
      resourceObject=null
      tiling=NoTile
      imageSize=<16 16>
      margin=<ltrb 0 0 0 0>
      outlineSettings {
       bUseBrushTransparency=false
       roundingType=HalfHeightRadius
       width=0
       color {colorUseRule=UseColor_Specified, specifiedColor=<rgba 0 0 0 0>}
       cornerRadii=<0 0 0 0>
      }
      tintColor {colorUseRule=UseColor_Specified, specifiedColor=<rgba 0.0044 0.2384 1 1>}
      uVRegion {bIsValid=false, max=<0 0>, min=<0 0>}
     }
     undeterminedImage {
      bIsDynamicallyLoaded=false
      drawAs=Image
      imageType=Vector
      mirroring=NoMirror
      resourceName=../../../Engine/Content/Slate/Starship/CoreWidgets/CheckBox/indeterminate.svg
      resourceObject=null
      tiling=NoTile
      imageSize=<16 16>
      margin=<ltrb 0 0 0 0>
      outlineSettings {
       bUseBrushTransparency=false
       roundingType=HalfHeightRadius
       width=0
       color {colorUseRule=UseColor_Specified, specifiedColor=<rgba 0 0 0 0>}
       cornerRadii=<0 0 0 0>
      }
      tintColor {colorUseRule=UseColor_Specified, specifiedColor=<rgba 0 0.162 0.7454 1>}
      uVRegion {bIsValid=false, max=<0 0>, min=<0 0>}
     }
     undeterminedPressedImage {
      bIsDynamicallyLoaded=false
      drawAs=Image
      imageType=Vector
      mirroring=NoMirror
      resourceName=../../../Engine/Content/Slate/Starship/CoreWidgets/CheckBox/indeterminate.svg
      resourceObject=null
      tiling=NoTile
      imageSize=<16 16>
      margin=<ltrb 0 0 0 0>
      outlineSettings {
       bUseBrushTransparency=false
       roundingType=HalfHeightRadius
       width=0
       color {colorUseRule=UseColor_Specified, specifiedColor=<rgba 0 0 0 0>}
       cornerRadii=<0 0 0 0>
      }
      tintColor {colorUseRule=UseColor_Specified, specifiedColor=<rgba 0 0.0802 0.3515 1>}
      uVRegion {bIsValid=false, max=<0 0>, min=<0 0>}
     }
    }
   }
   slot {class=HorizontalBoxSlot, properties {padding=<ltrb 0 0 5 0>}}
  }
  TextBlock_9 {
   widgetClassPath=&TextBlock
   parent=HorizontalBox_3
   widget=TextBlock_9
   properties {
    text=Include AnimSequences
    colorAndOpacity {colorUseRule=UseColor_Specified, specifiedColor=<rgba 0.0802 0.1981 0.0648 1>}
    font {
     bForceMonospaced=false
     bMaterialIsStencil=false
     fontMaterial=null
     fontObject=&Roboto
     letterSpacing=0
     monospacedWidth=1
     size=7.5
     skewAmount=0
     typefaceFontName=Bold
     outlineSettings {
      bApplyOutlineToDropShadows=false
      bMiteredCorners=false
      bSeparateFillAlpha=false
      outlineMaterial=null
      outlineSize=0
      outlineColor=<rgba 0 0 0 1>
     }
    }
   }
   slot {class=HorizontalBoxSlot, properties {verticalAlignment=VAlign_Center, padding=<ltrb 0 0 10 0>}}
  }
  IncludeAnimMontagesCheckbox {
   widgetClassPath=&EditorUtilityCheckBox
   bIsVariable=true
   parent=HorizontalBox_3
   widget=IncludeAnimMontagesCheckbox
   properties {
    checkedState=Checked
    widgetStyle {
     checkBoxType=CheckBox
     backgroundHoveredImage {
      bIsDynamicallyLoaded=false
      drawAs=RoundedBox
      imageType=NoImage
      mirroring=NoMirror
      resourceName=None
      resourceObject=null
      tiling=NoTile
      imageSize=<18 18>
      margin=<ltrb 0 0 0 0>
      outlineSettings {
       bUseBrushTransparency=false
       roundingType=FixedRadius
       width=1
       color {colorUseRule=UseColor_ColorTable, specifiedColor=<rgba 0 0 0 0>}
       cornerRadii=<3 3 3 3>
      }
      tintColor {colorUseRule=UseColor_Specified, specifiedColor=<rgba 0.0048 0.0048 0.0048 1>}
      uVRegion {bIsValid=false, max=<0 0>, min=<0 0>}
     }
     backgroundImage {
      bIsDynamicallyLoaded=false
      drawAs=RoundedBox
      imageType=NoImage
      mirroring=NoMirror
      resourceName=None
      resourceObject=null
      tiling=NoTile
      imageSize=<18 18>
      margin=<ltrb 0 0 0 0>
      outlineSettings {
       bUseBrushTransparency=false
       roundingType=FixedRadius
       width=1
       color {colorUseRule=UseColor_ColorTable, specifiedColor=<rgba 0 0 0 0>}
       cornerRadii=<3 3 3 3>
      }
      tintColor {colorUseRule=UseColor_Specified, specifiedColor=<rgba 0.0048 0.0048 0.0048 1>}
      uVRegion {bIsValid=false, max=<0 0>, min=<0 0>}
     }
     backgroundPressedImage {
      bIsDynamicallyLoaded=false
      drawAs=RoundedBox
      imageType=NoImage
      mirroring=NoMirror
      resourceName=None
      resourceObject=null
      tiling=NoTile
      imageSize=<18 18>
      margin=<ltrb 0 0 0 0>
      outlineSettings {
       bUseBrushTransparency=false
       roundingType=FixedRadius
       width=1
       color {colorUseRule=UseColor_ColorTable, specifiedColor=<rgba 0 0 0 0>}
       cornerRadii=<3 3 3 3>
      }
      tintColor {colorUseRule=UseColor_Specified, specifiedColor=<rgba 0.0048 0.0048 0.0048 1>}
      uVRegion {bIsValid=false, max=<0 0>, min=<0 0>}
     }
     borderBackgroundColor {colorUseRule=UseColor_Specified, specifiedColor=<rgba 1 1 1 1>}
     checkedForeground {colorUseRule=UseColor_Specified, specifiedColor=<rgba 1 1 1 1>}
     checkedHoveredForeground {colorUseRule=UseColor_Specified, specifiedColor=<rgba 1 1 1 1>}
     checkedHoveredImage {
      bIsDynamicallyLoaded=false
      drawAs=Image
      imageType=Vector
      mirroring=NoMirror
      resourceName=../../../Engine/Content/Slate/Starship/CoreWidgets/CheckBox/check.svg
      resourceObject=null
      tiling=NoTile
      imageSize=<16 16>
      margin=<ltrb 0 0 0 0>
      outlineSettings {
       bUseBrushTransparency=false
       roundingType=HalfHeightRadius
       width=0
       color {colorUseRule=UseColor_Specified, specifiedColor=<rgba 0 0 0 0>}
       cornerRadii=<0 0 0 0>
      }
      tintColor {colorUseRule=UseColor_Specified, specifiedColor=<rgba 0.0044 0.2384 1 1>}
      uVRegion {bIsValid=false, max=<0 0>, min=<0 0>}
     }
     checkedImage {
      bIsDynamicallyLoaded=false
      drawAs=Image
      imageType=Vector
      mirroring=NoMirror
      resourceName=../../../Engine/Content/Slate/Starship/CoreWidgets/CheckBox/check.svg
      resourceObject=null
      tiling=NoTile
      imageSize=<16 16>
      margin=<ltrb 0 0 0 0>
      outlineSettings {
       bUseBrushTransparency=false
       roundingType=HalfHeightRadius
       width=0
       color {colorUseRule=UseColor_Specified, specifiedColor=<rgba 0 0 0 0>}
       cornerRadii=<0 0 0 0>
      }
      tintColor {colorUseRule=UseColor_Specified, specifiedColor=<rgba 0 0.162 0.7454 1>}
      uVRegion {bIsValid=false, max=<0 0>, min=<0 0>}
     }
     checkedPressedForeground {colorUseRule=UseColor_Specified, specifiedColor=<rgba 1 1 1 1>}
     checkedPressedImage {
      bIsDynamicallyLoaded=false
      drawAs=Image
      imageType=Vector
      mirroring=NoMirror
      resourceName=../../../Engine/Content/Slate/Starship/CoreWidgets/CheckBox/check.svg
      resourceObject=null
      tiling=NoTile
      imageSize=<16 16>
      margin=<ltrb 0 0 0 0>
      outlineSettings {
       bUseBrushTransparency=false
       roundingType=HalfHeightRadius
       width=0
       color {colorUseRule=UseColor_Specified, specifiedColor=<rgba 0 0 0 0>}
       cornerRadii=<0 0 0 0>
      }
      tintColor {colorUseRule=UseColor_Specified, specifiedColor=<rgba 0 0.0802 0.3515 1>}
      uVRegion {bIsValid=false, max=<0 0>, min=<0 0>}
     }
     checkedSlateSound {resourceObject=null}
     foregroundColor {colorUseRule=UseColor_Specified, specifiedColor=<rgba 1 1 1 1>}
     hoveredForeground {colorUseRule=UseColor_Specified, specifiedColor=<rgba 1 1 1 1>}
     hoveredSlateSound {resourceObject=null}
     padding=<ltrb 0 0 0 0>
     pressedForeground {colorUseRule=UseColor_Specified, specifiedColor=<rgba 1 1 1 1>}
     uncheckedHoveredImage {
      bIsDynamicallyLoaded=false
      drawAs=NoDrawType
      imageType=NoImage
      mirroring=NoMirror
      resourceName=None
      resourceObject=null
      tiling=NoTile
      imageSize=<0 0>
      margin=<ltrb 0 0 0 0>
      outlineSettings {
       bUseBrushTransparency=false
       roundingType=HalfHeightRadius
       width=0
       color {colorUseRule=UseColor_Specified, specifiedColor=<rgba 0 0 0 0>}
       cornerRadii=<0 0 0 0>
      }
      tintColor {colorUseRule=UseColor_Specified, specifiedColor=<rgba 1 1 1 1>}
      uVRegion {bIsValid=false, max=<0 0>, min=<0 0>}
     }
     uncheckedImage {
      bIsDynamicallyLoaded=false
      drawAs=NoDrawType
      imageType=NoImage
      mirroring=NoMirror
      resourceName=None
      resourceObject=null
      tiling=NoTile
      imageSize=<0 0>
      margin=<ltrb 0 0 0 0>
      outlineSettings {
       bUseBrushTransparency=false
       roundingType=HalfHeightRadius
       width=0
       color {colorUseRule=UseColor_Specified, specifiedColor=<rgba 0 0 0 0>}
       cornerRadii=<0 0 0 0>
      }
      tintColor {colorUseRule=UseColor_Specified, specifiedColor=<rgba 1 1 1 1>}
      uVRegion {bIsValid=false, max=<0 0>, min=<0 0>}
     }
     uncheckedPressedImage {
      bIsDynamicallyLoaded=false
      drawAs=NoDrawType
      imageType=NoImage
      mirroring=NoMirror
      resourceName=None
      resourceObject=null
      tiling=NoTile
      imageSize=<0 0>
      margin=<ltrb 0 0 0 0>
      outlineSettings {
       bUseBrushTransparency=false
       roundingType=HalfHeightRadius
       width=0
       color {colorUseRule=UseColor_Specified, specifiedColor=<rgba 0 0 0 0>}
       cornerRadii=<0 0 0 0>
      }
      tintColor {colorUseRule=UseColor_Specified, specifiedColor=<rgba 1 1 1 1>}
      uVRegion {bIsValid=false, max=<0 0>, min=<0 0>}
     }
     uncheckedSlateSound {resourceObject=null}
     undeterminedForeground {colorUseRule=UseColor_Specified, specifiedColor=<rgba 1 1 1 1>}
     undeterminedHoveredImage {
      bIsDynamicallyLoaded=false
      drawAs=Image
      imageType=Vector
      mirroring=NoMirror
      resourceName=../../../Engine/Content/Slate/Starship/CoreWidgets/CheckBox/indeterminate.svg
      resourceObject=null
      tiling=NoTile
      imageSize=<16 16>
      margin=<ltrb 0 0 0 0>
      outlineSettings {
       bUseBrushTransparency=false
       roundingType=HalfHeightRadius
       width=0
       color {colorUseRule=UseColor_Specified, specifiedColor=<rgba 0 0 0 0>}
       cornerRadii=<0 0 0 0>
      }
      tintColor {colorUseRule=UseColor_Specified, specifiedColor=<rgba 0.0044 0.2384 1 1>}
      uVRegion {bIsValid=false, max=<0 0>, min=<0 0>}
     }
     undeterminedImage {
      bIsDynamicallyLoaded=false
      drawAs=Image
      imageType=Vector
      mirroring=NoMirror
      resourceName=../../../Engine/Content/Slate/Starship/CoreWidgets/CheckBox/indeterminate.svg
      resourceObject=null
      tiling=NoTile
      imageSize=<16 16>
      margin=<ltrb 0 0 0 0>
      outlineSettings {
       bUseBrushTransparency=false
       roundingType=HalfHeightRadius
       width=0
       color {colorUseRule=UseColor_Specified, specifiedColor=<rgba 0 0 0 0>}
       cornerRadii=<0 0 0 0>
      }
      tintColor {colorUseRule=UseColor_Specified, specifiedColor=<rgba 0 0.162 0.7454 1>}
      uVRegion {bIsValid=false, max=<0 0>, min=<0 0>}
     }
     undeterminedPressedImage {
      bIsDynamicallyLoaded=false
      drawAs=Image
      imageType=Vector
      mirroring=NoMirror
      resourceName=../../../Engine/Content/Slate/Starship/CoreWidgets/CheckBox/indeterminate.svg
      resourceObject=null
      tiling=NoTile
      imageSize=<16 16>
      margin=<ltrb 0 0 0 0>
      outlineSettings {
       bUseBrushTransparency=false
       roundingType=HalfHeightRadius
       width=0
       color {colorUseRule=UseColor_Specified, specifiedColor=<rgba 0 0 0 0>}
       cornerRadii=<0 0 0 0>
      }
      tintColor {colorUseRule=UseColor_Specified, specifiedColor=<rgba 0 0.0802 0.3515 1>}
      uVRegion {bIsValid=false, max=<0 0>, min=<0 0>}
     }
    }
   }
   slot {class=HorizontalBoxSlot, properties {padding=<ltrb 0 0 5 0>}}
  }
  TextBlock_1 {
   widgetClassPath=&TextBlock
   parent=HorizontalBox_3
   widget=TextBlock_1
   properties {
    text=Include AnimMontages
    colorAndOpacity {colorUseRule=UseColor_Specified, specifiedColor=<rgba 0.1274 0.1274 1 1>}
    font {
     bForceMonospaced=false
     bMaterialIsStencil=false
     fontMaterial=null
     fontObject=&Roboto
     letterSpacing=0
     monospacedWidth=1
     size=7.5
     skewAmount=0
     typefaceFontName=Bold
     outlineSettings {
      bApplyOutlineToDropShadows=false
      bMiteredCorners=false
      bSeparateFillAlpha=false
      outlineMaterial=null
      outlineSize=0
      outlineColor=<rgba 0 0 0 1>
     }
    }
   }
   slot {class=HorizontalBoxSlot, properties {verticalAlignment=VAlign_Center}}
  }
  EditorUtilityScrollBox_0 {
   widgetClassPath=&EditorUtilityScrollBox
   parent=VerticalBox_0
   widget=EditorUtilityScrollBox_0
   slot {class=VerticalBoxSlot, properties {size {sizeRule=Fill, value=1}}}
  }
  Border_0 {
   widgetClassPath=&Border
   parent=EditorUtilityScrollBox_0
   widget=Border_0
   properties {brushColor=<rgba 0.005 0.005 0.005 1>, padding=<ltrb 5 5 5 5>}
   slot {class=ScrollBoxSlot, properties {size {sizeRule=Fill, value=1}}}
  }
  ListView {
   widgetClassPath=/Script/Blutility.EditorUtilityListView
   bIsVariable=true
   parent=Border_0
   widget=ListView
   properties {entryWidgetClass=/AnimInEdTools/AnimBatchDuplicate/Widgets/EUW_AnimBatchDup_Entry.EUW_AnimBatchDup_Entry_C}
   slot {class=BorderSlot, properties {padding=<ltrb 5 5 5 5>}}
  }
  VerticalBox_4 {widgetClassPath=&VerticalBox, parent=VerticalBox_0, widget=VerticalBox_4, slot {class=VerticalBoxSlot}}
  IKRetargeterBorder {
   widgetClassPath=&Border
   bIsVariable=true
   parent=VerticalBox_4
   widget=IKRetargeterBorder
   properties {
    horizontalAlignment=HAlign_Center
    background {
     bIsDynamicallyLoaded=false
     drawAs=Image
     imageType=NoImage
     mirroring=NoMirror
     resourceName=None
     resourceObject=null
     tiling=NoTile
     imageSize=<32 32>
     margin=<ltrb 0 0 0 0>
     outlineSettings {
      bUseBrushTransparency=false
      roundingType=HalfHeightRadius
      width=0
      color {colorUseRule=UseColor_Specified, specifiedColor=<rgba 0 0 0 0>}
      cornerRadii=<0 0 0 0>
     }
     tintColor {colorUseRule=UseColor_Specified, specifiedColor=<rgba 0.0194 0.0194 0.0194 1>}
     uVRegion {bIsValid=false, max=<0 0>, min=<0 0>}
    }
    padding=<ltrb 0 10 0 10>
   }
   slot {class=VerticalBoxSlot, properties {size {sizeRule=Fill, value=1}}}
  }
  HorizontalBox_4 {
   widgetClassPath=&HorizontalBox
   parent=IKRetargeterBorder
   widget=HorizontalBox_4
   slot {class=BorderSlot, properties {horizontalAlignment=HAlign_Center, padding=<ltrb 0 10 0 10>}}
  }
  BakeToNewCharacterCheckbox {
   widgetClassPath=&EditorUtilityCheckBox
   bIsVariable=true
   parent=HorizontalBox_4
   widget=BakeToNewCharacterCheckbox
   properties {widgetStyle {
     checkBoxType=CheckBox
     backgroundHoveredImage {
      bIsDynamicallyLoaded=false
      drawAs=RoundedBox
      imageType=NoImage
      mirroring=NoMirror
      resourceName=None
      resourceObject=null
      tiling=NoTile
      imageSize=<18 18>
      margin=<ltrb 0 0 0 0>
      outlineSettings {
       bUseBrushTransparency=false
       roundingType=FixedRadius
       width=1
       color {colorUseRule=UseColor_ColorTable, specifiedColor=<rgba 0 0 0 0>}
       cornerRadii=<3 3 3 3>
      }
      tintColor {colorUseRule=UseColor_Specified, specifiedColor=<rgba 0.0048 0.0048 0.0048 1>}
      uVRegion {bIsValid=false, max=<0 0>, min=<0 0>}
     }
     backgroundImage {
      bIsDynamicallyLoaded=false
      drawAs=RoundedBox
      imageType=NoImage
      mirroring=NoMirror
      resourceName=None
      resourceObject=null
      tiling=NoTile
      imageSize=<18 18>
      margin=<ltrb 0 0 0 0>
      outlineSettings {
       bUseBrushTransparency=false
       roundingType=FixedRadius
       width=1
       color {colorUseRule=UseColor_ColorTable, specifiedColor=<rgba 0 0 0 0>}
       cornerRadii=<3 3 3 3>
      }
      tintColor {colorUseRule=UseColor_Specified, specifiedColor=<rgba 0.0048 0.0048 0.0048 1>}
      uVRegion {bIsValid=false, max=<0 0>, min=<0 0>}
     }
     backgroundPressedImage {
      bIsDynamicallyLoaded=false
      drawAs=RoundedBox
      imageType=NoImage
      mirroring=NoMirror
      resourceName=None
      resourceObject=null
      tiling=NoTile
      imageSize=<18 18>
      margin=<ltrb 0 0 0 0>
      outlineSettings {
       bUseBrushTransparency=false
       roundingType=FixedRadius
       width=1
       color {colorUseRule=UseColor_ColorTable, specifiedColor=<rgba 0 0 0 0>}
       cornerRadii=<3 3 3 3>
      }
      tintColor {colorUseRule=UseColor_Specified, specifiedColor=<rgba 0.0048 0.0048 0.0048 1>}
      uVRegion {bIsValid=false, max=<0 0>, min=<0 0>}
     }
     borderBackgroundColor {colorUseRule=UseColor_Specified, specifiedColor=<rgba 1 1 1 1>}
     checkedForeground {colorUseRule=UseColor_Specified, specifiedColor=<rgba 1 1 1 1>}
     checkedHoveredForeground {colorUseRule=UseColor_Specified, specifiedColor=<rgba 1 1 1 1>}
     checkedHoveredImage {
      bIsDynamicallyLoaded=false
      drawAs=Image
      imageType=Vector
      mirroring=NoMirror
      resourceName=../../../Engine/Content/Slate/Starship/CoreWidgets/CheckBox/check.svg
      resourceObject=null
      tiling=NoTile
      imageSize=<16 16>
      margin=<ltrb 0 0 0 0>
      outlineSettings {
       bUseBrushTransparency=false
       roundingType=HalfHeightRadius
       width=0
       color {colorUseRule=UseColor_Specified, specifiedColor=<rgba 0 0 0 0>}
       cornerRadii=<0 0 0 0>
      }
      tintColor {colorUseRule=UseColor_Specified, specifiedColor=<rgba 0.0044 0.2384 1 1>}
      uVRegion {bIsValid=false, max=<0 0>, min=<0 0>}
     }
     checkedImage {
      bIsDynamicallyLoaded=false
      drawAs=Image
      imageType=Vector
      mirroring=NoMirror
      resourceName=../../../Engine/Content/Slate/Starship/CoreWidgets/CheckBox/check.svg
      resourceObject=null
      tiling=NoTile
      imageSize=<16 16>
      margin=<ltrb 0 0 0 0>
      outlineSettings {
       bUseBrushTransparency=false
       roundingType=HalfHeightRadius
       width=0
       color {colorUseRule=UseColor_Specified, specifiedColor=<rgba 0 0 0 0>}
       cornerRadii=<0 0 0 0>
      }
      tintColor {colorUseRule=UseColor_Specified, specifiedColor=<rgba 0 0.162 0.7454 1>}
      uVRegion {bIsValid=false, max=<0 0>, min=<0 0>}
     }
     checkedPressedForeground {colorUseRule=UseColor_Specified, specifiedColor=<rgba 1 1 1 1>}
     checkedPressedImage {
      bIsDynamicallyLoaded=false
      drawAs=Image
      imageType=Vector
      mirroring=NoMirror
      resourceName=../../../Engine/Content/Slate/Starship/CoreWidgets/CheckBox/check.svg
      resourceObject=null
      tiling=NoTile
      imageSize=<16 16>
      margin=<ltrb 0 0 0 0>
      outlineSettings {
       bUseBrushTransparency=false
       roundingType=HalfHeightRadius
       width=0
       color {colorUseRule=UseColor_Specified, specifiedColor=<rgba 0 0 0 0>}
       cornerRadii=<0 0 0 0>
      }
      tintColor {colorUseRule=UseColor_Specified, specifiedColor=<rgba 0 0.0802 0.3515 1>}
      uVRegion {bIsValid=false, max=<0 0>, min=<0 0>}
     }
     checkedSlateSound {resourceObject=null}
     foregroundColor {colorUseRule=UseColor_Specified, specifiedColor=<rgba 1 1 1 1>}
     hoveredForeground {colorUseRule=UseColor_Specified, specifiedColor=<rgba 1 1 1 1>}
     hoveredSlateSound {resourceObject=null}
     padding=<ltrb 0 0 0 0>
     pressedForeground {colorUseRule=UseColor_Specified, specifiedColor=<rgba 1 1 1 1>}
     uncheckedHoveredImage {
      bIsDynamicallyLoaded=false
      drawAs=NoDrawType
      imageType=NoImage
      mirroring=NoMirror
      resourceName=None
      resourceObject=null
      tiling=NoTile
      imageSize=<0 0>
      margin=<ltrb 0 0 0 0>
      outlineSettings {
       bUseBrushTransparency=false
       roundingType=HalfHeightRadius
       width=0
       color {colorUseRule=UseColor_Specified, specifiedColor=<rgba 0 0 0 0>}
       cornerRadii=<0 0 0 0>
      }
      tintColor {colorUseRule=UseColor_Specified, specifiedColor=<rgba 1 1 1 1>}
      uVRegion {bIsValid=false, max=<0 0>, min=<0 0>}
     }
     uncheckedImage {
      bIsDynamicallyLoaded=false
      drawAs=NoDrawType
      imageType=NoImage
      mirroring=NoMirror
      resourceName=None
      resourceObject=null
      tiling=NoTile
      imageSize=<0 0>
      margin=<ltrb 0 0 0 0>
      outlineSettings {
       bUseBrushTransparency=false
       roundingType=HalfHeightRadius
       width=0
       color {colorUseRule=UseColor_Specified, specifiedColor=<rgba 0 0 0 0>}
       cornerRadii=<0 0 0 0>
      }
      tintColor {colorUseRule=UseColor_Specified, specifiedColor=<rgba 1 1 1 1>}
      uVRegion {bIsValid=false, max=<0 0>, min=<0 0>}
     }
     uncheckedPressedImage {
      bIsDynamicallyLoaded=false
      drawAs=NoDrawType
      imageType=NoImage
      mirroring=NoMirror
      resourceName=None
      resourceObject=null
      tiling=NoTile
      imageSize=<0 0>
      margin=<ltrb 0 0 0 0>
      outlineSettings {
       bUseBrushTransparency=false
       roundingType=HalfHeightRadius
       width=0
       color {colorUseRule=UseColor_Specified, specifiedColor=<rgba 0 0 0 0>}
       cornerRadii=<0 0 0 0>
      }
      tintColor {colorUseRule=UseColor_Specified, specifiedColor=<rgba 1 1 1 1>}
      uVRegion {bIsValid=false, max=<0 0>, min=<0 0>}
     }
     uncheckedSlateSound {resourceObject=null}
     undeterminedForeground {colorUseRule=UseColor_Specified, specifiedColor=<rgba 1 1 1 1>}
     undeterminedHoveredImage {
      bIsDynamicallyLoaded=false
      drawAs=Image
      imageType=Vector
      mirroring=NoMirror
      resourceName=../../../Engine/Content/Slate/Starship/CoreWidgets/CheckBox/indeterminate.svg
      resourceObject=null
      tiling=NoTile
      imageSize=<16 16>
      margin=<ltrb 0 0 0 0>
      outlineSettings {
       bUseBrushTransparency=false
       roundingType=HalfHeightRadius
       width=0
       color {colorUseRule=UseColor_Specified, specifiedColor=<rgba 0 0 0 0>}
       cornerRadii=<0 0 0 0>
      }
      tintColor {colorUseRule=UseColor_Specified, specifiedColor=<rgba 0.0044 0.2384 1 1>}
      uVRegion {bIsValid=false, max=<0 0>, min=<0 0>}
     }
     undeterminedImage {
      bIsDynamicallyLoaded=false
      drawAs=Image
      imageType=Vector
      mirroring=NoMirror
      resourceName=../../../Engine/Content/Slate/Starship/CoreWidgets/CheckBox/indeterminate.svg
      resourceObject=null
      tiling=NoTile
      imageSize=<16 16>
      margin=<ltrb 0 0 0 0>
      outlineSettings {
       bUseBrushTransparency=false
       roundingType=HalfHeightRadius
       width=0
       color {colorUseRule=UseColor_Specified, specifiedColor=<rgba 0 0 0 0>}
       cornerRadii=<0 0 0 0>
      }
      tintColor {colorUseRule=UseColor_Specified, specifiedColor=<rgba 0 0.162 0.7454 1>}
      uVRegion {bIsValid=false, max=<0 0>, min=<0 0>}
     }
     undeterminedPressedImage {
      bIsDynamicallyLoaded=false
      drawAs=Image
      imageType=Vector
      mirroring=NoMirror
      resourceName=../../../Engine/Content/Slate/Starship/CoreWidgets/CheckBox/indeterminate.svg
      resourceObject=null
      tiling=NoTile
      imageSize=<16 16>
      margin=<ltrb 0 0 0 0>
      outlineSettings {
       bUseBrushTransparency=false
       roundingType=HalfHeightRadius
       width=0
       color {colorUseRule=UseColor_Specified, specifiedColor=<rgba 0 0 0 0>}
       cornerRadii=<0 0 0 0>
      }
      tintColor {colorUseRule=UseColor_Specified, specifiedColor=<rgba 0 0.0802 0.3515 1>}
      uVRegion {bIsValid=false, max=<0 0>, min=<0 0>}
     }
    }}
   slot {class=HorizontalBoxSlot, properties {horizontalAlignment=HAlign_Left, padding=<ltrb 0 0 10 0>}}
  }
  VerticalBox_5 {
   widgetClassPath=&VerticalBox
   parent=HorizontalBox_4
   widget=VerticalBox_5
   slot {class=HorizontalBoxSlot, properties {horizontalAlignment=HAlign_Left}}
  }
  TextBlock_92 {
   widgetClassPath=&TextBlock
   parent=VerticalBox_5
   widget=TextBlock_92
   properties {
    text=Bake animations to new character
    font {
     bForceMonospaced=false
     bMaterialIsStencil=false
     fontMaterial=null
     fontObject=&Roboto
     letterSpacing=0
     monospacedWidth=1
     size=10.5
     skewAmount=0
     typefaceFontName=Bold
     outlineSettings {
      bApplyOutlineToDropShadows=false
      bMiteredCorners=false
      bSeparateFillAlpha=false
      outlineMaterial=null
      outlineSize=0
      outlineColor=<rgba 0 0 0 1>
     }
    }
   }
   slot {class=VerticalBoxSlot}
  }
  TextBlock_186 {
   widgetClassPath=&TextBlock
   parent=VerticalBox_5
   widget=TextBlock_186
   properties {
    text=Experimental. Bake the source animations to a new Character's rig using an existing IKRetargeter asset. Only works with AC inputs.
    font {
     bForceMonospaced=false
     bMaterialIsStencil=false
     fontMaterial=null
     fontObject=&Roboto
     letterSpacing=0
     monospacedWidth=1
     size=9
     skewAmount=0
     typefaceFontName=Regular
     outlineSettings {
      bApplyOutlineToDropShadows=false
      bMiteredCorners=false
      bSeparateFillAlpha=false
      outlineMaterial=null
      outlineSize=0
      outlineColor=<rgba 0 0 0 1>
     }
    }
   }
   slot {class=VerticalBoxSlot}
  }
  IKRetargeterPropertyView {
   widgetClassPath=&SinglePropertyView
   parent=VerticalBox_5
   widget=IKRetargeterPropertyView
   properties {propertyName=IK Retargeter}
   slot {class=VerticalBoxSlot, properties {padding=<ltrb 0 5 0 0>, size {sizeRule=Fill, value=1}}}
  }
  HorizontalBox_8 {
   widgetClassPath=&HorizontalBox
   parent=VerticalBox_5
   widget=HorizontalBox_8
   properties {toolTipText="Off (Recommended) - Retarget same as source\r\nOn - Retarget as additive delta"}
   slot {class=VerticalBoxSlot, properties {padding=<ltrb 0 10 0 5>}}
  }
  EditorUtilityCheckBox_0 {
   widgetClassPath=&EditorUtilityCheckBox
   bIsVariable=true
   parent=HorizontalBox_8
   widget=EditorUtilityCheckBox_0
   properties {widgetStyle {
     checkBoxType=CheckBox
     backgroundHoveredImage {
      bIsDynamicallyLoaded=false
      drawAs=RoundedBox
      imageType=NoImage
      mirroring=NoMirror
      resourceName=None
      resourceObject=null
      tiling=NoTile
      imageSize=<18 18>
      margin=<ltrb 0 0 0 0>
      outlineSettings {
       bUseBrushTransparency=false
       roundingType=FixedRadius
       width=1
       color {colorUseRule=UseColor_ColorTable, specifiedColor=<rgba 0 0 0 0>}
       cornerRadii=<3 3 3 3>
      }
      tintColor {colorUseRule=UseColor_Specified, specifiedColor=<rgba 0.0048 0.0048 0.0048 1>}
      uVRegion {bIsValid=false, max=<0 0>, min=<0 0>}
     }
     backgroundImage {
      bIsDynamicallyLoaded=false
      drawAs=RoundedBox
      imageType=NoImage
      mirroring=NoMirror
      resourceName=None
      resourceObject=null
      tiling=NoTile
      imageSize=<18 18>
      margin=<ltrb 0 0 0 0>
      outlineSettings {
       bUseBrushTransparency=false
       roundingType=FixedRadius
       width=1
       color {colorUseRule=UseColor_ColorTable, specifiedColor=<rgba 0 0 0 0>}
       cornerRadii=<3 3 3 3>
      }
      tintColor {colorUseRule=UseColor_Specified, specifiedColor=<rgba 0.0048 0.0048 0.0048 1>}
      uVRegion {bIsValid=false, max=<0 0>, min=<0 0>}
     }
     backgroundPressedImage {
      bIsDynamicallyLoaded=false
      drawAs=RoundedBox
      imageType=NoImage
      mirroring=NoMirror
      resourceName=None
      resourceObject=null
      tiling=NoTile
      imageSize=<18 18>
      margin=<ltrb 0 0 0 0>
      outlineSettings {
       bUseBrushTransparency=false
       roundingType=FixedRadius
       width=1
       color {colorUseRule=UseColor_ColorTable, specifiedColor=<rgba 0 0 0 0>}
       cornerRadii=<3 3 3 3>
      }
      tintColor {colorUseRule=UseColor_Specified, specifiedColor=<rgba 0.0048 0.0048 0.0048 1>}
      uVRegion {bIsValid=false, max=<0 0>, min=<0 0>}
     }
     borderBackgroundColor {colorUseRule=UseColor_Specified, specifiedColor=<rgba 1 1 1 1>}
     checkedForeground {colorUseRule=UseColor_Specified, specifiedColor=<rgba 1 1 1 1>}
     checkedHoveredForeground {colorUseRule=UseColor_Specified, specifiedColor=<rgba 1 1 1 1>}
     checkedHoveredImage {
      bIsDynamicallyLoaded=false
      drawAs=Image
      imageType=Vector
      mirroring=NoMirror
      resourceName=../../../Engine/Content/Slate/Starship/CoreWidgets/CheckBox/check.svg
      resourceObject=null
      tiling=NoTile
      imageSize=<16 16>
      margin=<ltrb 0 0 0 0>
      outlineSettings {
       bUseBrushTransparency=false
       roundingType=HalfHeightRadius
       width=0
       color {colorUseRule=UseColor_Specified, specifiedColor=<rgba 0 0 0 0>}
       cornerRadii=<0 0 0 0>
      }
      tintColor {colorUseRule=UseColor_Specified, specifiedColor=<rgba 0.0044 0.2384 1 1>}
      uVRegion {bIsValid=false, max=<0 0>, min=<0 0>}
     }
     checkedImage {
      bIsDynamicallyLoaded=false
      drawAs=Image
      imageType=Vector
      mirroring=NoMirror
      resourceName=../../../Engine/Content/Slate/Starship/CoreWidgets/CheckBox/check.svg
      resourceObject=null
      tiling=NoTile
      imageSize=<16 16>
      margin=<ltrb 0 0 0 0>
      outlineSettings {
       bUseBrushTransparency=false
       roundingType=HalfHeightRadius
       width=0
       color {colorUseRule=UseColor_Specified, specifiedColor=<rgba 0 0 0 0>}
       cornerRadii=<0 0 0 0>
      }
      tintColor {colorUseRule=UseColor_Specified, specifiedColor=<rgba 0 0.162 0.7454 1>}
      uVRegion {bIsValid=false, max=<0 0>, min=<0 0>}
     }
     checkedPressedForeground {colorUseRule=UseColor_Specified, specifiedColor=<rgba 1 1 1 1>}
     checkedPressedImage {
      bIsDynamicallyLoaded=false
      drawAs=Image
      imageType=Vector
      mirroring=NoMirror
      resourceName=../../../Engine/Content/Slate/Starship/CoreWidgets/CheckBox/check.svg
      resourceObject=null
      tiling=NoTile
      imageSize=<16 16>
      margin=<ltrb 0 0 0 0>
      outlineSettings {
       bUseBrushTransparency=false
       roundingType=HalfHeightRadius
       width=0
       color {colorUseRule=UseColor_Specified, specifiedColor=<rgba 0 0 0 0>}
       cornerRadii=<0 0 0 0>
      }
      tintColor {colorUseRule=UseColor_Specified, specifiedColor=<rgba 0 0.0802 0.3515 1>}
      uVRegion {bIsValid=false, max=<0 0>, min=<0 0>}
     }
     checkedSlateSound {resourceObject=null}
     foregroundColor {colorUseRule=UseColor_Specified, specifiedColor=<rgba 1 1 1 1>}
     hoveredForeground {colorUseRule=UseColor_Specified, specifiedColor=<rgba 1 1 1 1>}
     hoveredSlateSound {resourceObject=null}
     padding=<ltrb 0 0 0 0>
     pressedForeground {colorUseRule=UseColor_Specified, specifiedColor=<rgba 1 1 1 1>}
     uncheckedHoveredImage {
      bIsDynamicallyLoaded=false
      drawAs=NoDrawType
      imageType=NoImage
      mirroring=NoMirror
      resourceName=None
      resourceObject=null
      tiling=NoTile
      imageSize=<0 0>
      margin=<ltrb 0 0 0 0>
      outlineSettings {
       bUseBrushTransparency=false
       roundingType=HalfHeightRadius
       width=0
       color {colorUseRule=UseColor_Specified, specifiedColor=<rgba 0 0 0 0>}
       cornerRadii=<0 0 0 0>
      }
      tintColor {colorUseRule=UseColor_Specified, specifiedColor=<rgba 1 1 1 1>}
      uVRegion {bIsValid=false, max=<0 0>, min=<0 0>}
     }
     uncheckedImage {
      bIsDynamicallyLoaded=false
      drawAs=NoDrawType
      imageType=NoImage
      mirroring=NoMirror
      resourceName=None
      resourceObject=null
      tiling=NoTile
      imageSize=<0 0>
      margin=<ltrb 0 0 0 0>
      outlineSettings {
       bUseBrushTransparency=false
       roundingType=HalfHeightRadius
       width=0
       color {colorUseRule=UseColor_Specified, specifiedColor=<rgba 0 0 0 0>}
       cornerRadii=<0 0 0 0>
      }
      tintColor {colorUseRule=UseColor_Specified, specifiedColor=<rgba 1 1 1 1>}
      uVRegion {bIsValid=false, max=<0 0>, min=<0 0>}
     }
     uncheckedPressedImage {
      bIsDynamicallyLoaded=false
      drawAs=NoDrawType
      imageType=NoImage
      mirroring=NoMirror
      resourceName=None
      resourceObject=null
      tiling=NoTile
      imageSize=<0 0>
      margin=<ltrb 0 0 0 0>
      outlineSettings {
       bUseBrushTransparency=false
       roundingType=HalfHeightRadius
       width=0
       color {colorUseRule=UseColor_Specified, specifiedColor=<rgba 0 0 0 0>}
       cornerRadii=<0 0 0 0>
      }
      tintColor {colorUseRule=UseColor_Specified, specifiedColor=<rgba 1 1 1 1>}
      uVRegion {bIsValid=false, max=<0 0>, min=<0 0>}
     }
     uncheckedSlateSound {resourceObject=null}
     undeterminedForeground {colorUseRule=UseColor_Specified, specifiedColor=<rgba 1 1 1 1>}
     undeterminedHoveredImage {
      bIsDynamicallyLoaded=false
      drawAs=Image
      imageType=Vector
      mirroring=NoMirror
      resourceName=../../../Engine/Content/Slate/Starship/CoreWidgets/CheckBox/indeterminate.svg
      resourceObject=null
      tiling=NoTile
      imageSize=<16 16>
      margin=<ltrb 0 0 0 0>
      outlineSettings {
       bUseBrushTransparency=false
       roundingType=HalfHeightRadius
       width=0
       color {colorUseRule=UseColor_Specified, specifiedColor=<rgba 0 0 0 0>}
       cornerRadii=<0 0 0 0>
      }
      tintColor {colorUseRule=UseColor_Specified, specifiedColor=<rgba 0.0044 0.2384 1 1>}
      uVRegion {bIsValid=false, max=<0 0>, min=<0 0>}
     }
     undeterminedImage {
      bIsDynamicallyLoaded=false
      drawAs=Image
      imageType=Vector
      mirroring=NoMirror
      resourceName=../../../Engine/Content/Slate/Starship/CoreWidgets/CheckBox/indeterminate.svg
      resourceObject=null
      tiling=NoTile
      imageSize=<16 16>
      margin=<ltrb 0 0 0 0>
      outlineSettings {
       bUseBrushTransparency=false
       roundingType=HalfHeightRadius
       width=0
       color {colorUseRule=UseColor_Specified, specifiedColor=<rgba 0 0 0 0>}
       cornerRadii=<0 0 0 0>
      }
      tintColor {colorUseRule=UseColor_Specified, specifiedColor=<rgba 0 0.162 0.7454 1>}
      uVRegion {bIsValid=false, max=<0 0>, min=<0 0>}
     }
     undeterminedPressedImage {
      bIsDynamicallyLoaded=false
      drawAs=Image
      imageType=Vector
      mirroring=NoMirror
      resourceName=../../../Engine/Content/Slate/Starship/CoreWidgets/CheckBox/indeterminate.svg
      resourceObject=null
      tiling=NoTile
      imageSize=<16 16>
      margin=<ltrb 0 0 0 0>
      outlineSettings {
       bUseBrushTransparency=false
       roundingType=HalfHeightRadius
       width=0
       color {colorUseRule=UseColor_Specified, specifiedColor=<rgba 0 0 0 0>}
       cornerRadii=<0 0 0 0>
      }
      tintColor {colorUseRule=UseColor_Specified, specifiedColor=<rgba 0 0.0802 0.3515 1>}
      uVRegion {bIsValid=false, max=<0 0>, min=<0 0>}
     }
    }}
   slot {class=HorizontalBoxSlot, properties {padding=<ltrb 0 0 5 0>}}
  }
  TextBlock_161 {
   widgetClassPath=&TextBlock
   parent=HorizontalBox_8
   widget=TextBlock_161
   properties {
    text=Retain Additive Flags
    font {
     bForceMonospaced=false
     bMaterialIsStencil=false
     fontMaterial=null
     fontObject=&Roboto
     letterSpacing=0
     monospacedWidth=1
     size=9
     skewAmount=0
     typefaceFontName=Bold
     outlineSettings {
      bApplyOutlineToDropShadows=false
      bMiteredCorners=false
      bSeparateFillAlpha=false
      outlineMaterial=null
      outlineSize=0
      outlineColor=<rgba 0 0 0 1>
     }
    }
   }
   slot {class=HorizontalBoxSlot, properties {verticalAlignment=VAlign_Center}}
  }
  DuplicateButton {widgetClassPath=&EditorUtilityButton, parent=VerticalBox_4, widget=DuplicateButton, slot {class=VerticalBoxSlot}}
  TextBlock_0 {
   widgetClassPath=&TextBlock
   parent=DuplicateButton
   widget=TextBlock_0
   properties {
    text=Duplicate Selected Animations
    font {
     bForceMonospaced=false
     bMaterialIsStencil=false
     fontMaterial=null
     fontObject=&Roboto
     letterSpacing=0
     monospacedWidth=1
     size=16.5
     skewAmount=0
     typefaceFontName=Bold
     outlineSettings {
      bApplyOutlineToDropShadows=false
      bMiteredCorners=false
      bSeparateFillAlpha=false
      outlineMaterial=null
      outlineSize=0
      outlineColor=<rgba 0 0 0 1>
     }
    }
   }
   slot {class=ButtonSlot, properties {padding=<ltrb 0 10 0 10>}}
  }
  LogBorder {
   widgetClassPath=&Border
   bIsVariable=true
   parent=VerticalBox_4
   widget=LogBorder
   properties {brushColor=<rgba 0.0048 0.0048 0.0048 1>, padding=<ltrb 20 5 20 5>}
   slot {class=VerticalBoxSlot}
  }
  LogText {
   widgetClassPath=&TextBlock
   bIsVariable=true
   parent=LogBorder
   widget=LogText
   properties {
    text=Tool Log
    font {
     bForceMonospaced=false
     bMaterialIsStencil=false
     fontMaterial=null
     fontObject=&Roboto
     letterSpacing=0
     monospacedWidth=1
     size=12
     skewAmount=0
     typefaceFontName=Default
     outlineSettings {
      bApplyOutlineToDropShadows=false
      bMiteredCorners=false
      bSeparateFillAlpha=false
      outlineMaterial=null
      outlineSize=0
      outlineColor=<rgba 0 0 0 1>
     }
    }
   }
   slot {class=BorderSlot, properties {padding=<ltrb 20 5 20 5>}}
  }
 ]
}
