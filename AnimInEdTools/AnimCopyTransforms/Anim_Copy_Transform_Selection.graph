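# Editor utility widget that serves as a selectable list row in the AnimInEdTools copy-transforms tool, made of a text label inside a Border.
# When assigned a list item, it casts it to BP_AnimCopy_ListItem_data, copies its text into Input Text and keeps a soft reference to it.
# Selection changes recolor the Border to bright cyan when selected and a transparent dark tint otherwise.
# The graph shown never writes to the EntryText label directly.
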EditorUtilityWidgetBlueprint /AnimInEdTools/AnimCopyTransforms/Anim_Copy_Transform_Selection
graphs [EventGraph {nodes [
   n1 = Event OnListItemObjectSet()
   > n2 = Cast To BP_AnimCopy_ListItem_data!(execute, Object=@n1.ListItemObject) -> (then, CastFailed, AsBP Anim Copy List Item Data:BP Anim Copy List Item Data Object Reference) {is=title K2Node_DynamicCast}
   > Set Input Text(Input Text:Text=Get Input Text!(self:BP Anim Copy List Item Data Object Reference=@n2) -> (Input Text:Text)) -> (Output_Get:Text)
   > Set Entry Data BP(Entry Data BP:BP Anim Copy List Item Data Soft Object Reference=Make Soft Reference!(Input:BP Anim Copy List Item Data Object Reference=@n2) -> (Output:BP Anim Copy List Item Data Soft Object Reference)) -> (Output_Get:BP Anim Copy List Item Data Soft Object Reference)
   n3 = Event On Item Selection Changed!() -> (OutputDelegate, then, bIsSelected:bool) {is=title K2Node_Event}
   > SetBrushColor(
    self=Get Border_0() -> (Border_0:Border Object Reference)
    InBrushColor=SelectColor(A=(R=2.5,G=10,B=15,A=1), B=(R=0.0086,G=0.0086,B=0.0086,A=0), bPickA=@n3.bIsSelected)
   )
   Event PreConstruct()
   Event Construct()
   Event Tick()
  ]}]
variables [
 Entry Data BP {type=BP_AnimCopy_ListItem_data_C, default=null}
 Input Text {type=string, default=---}
]
widgets {
 info {rootWidgetClass=&HorizontalBox, widgetCount=3}
 widgets [
  HorizontalBox_50 {widgetClassPath=&HorizontalBox, bIsVariable=true, widget=HorizontalBox_50}
  Border_0 {
   widgetClassPath=&Border
   bIsVariable=true
   parent=HorizontalBox_50
   widget=Border_0
   properties {
    background {
     bIsDynamicallyLoaded=false
     drawAs=Image
     imageType=NoImage
     mirroring=NoMirror
     resourceName=None
     resourceObject=null
     tiling=NoTile
     imageSize=<32 32>
     margin=<ltrb 0 0 0 0>
     outlineSettings {
      bUseBrushTransparency=false
      roundingType=HalfHeightRadius
      width=0
      color {colorUseRule=UseColor_Specified, specifiedColor=<rgba 0 0 0 0>}
      cornerRadii=<0 0 0 0>
     }
     tintColor {colorUseRule=UseColor_Specified, specifiedColor=<rgba 0.0086 0.0086 0.0086 1>}
     uVRegion {bIsValid=false, max=<0 0>, min=<0 0>}
    }
    brushColor=<rgba 1 1 1 0>
   }
   slot {class=HorizontalBoxSlot, properties {size {sizeRule=Fill, value=1}}}
  }
  EntryText {
   widgetClassPath=&TextBlock
   bIsVariable=true
   parent=Border_0
   widget=EntryText
   properties {
    text=--
    font {
     bForceMonospaced=false
     bMaterialIsStencil=false
     fontMaterial=null
     fontObject=&Roboto
     letterSpacing=0
     monospacedWidth=1
     size=12
     skewAmount=0
     typefaceFontName=Light
     outlineSettings {
      bApplyOutlineToDropShadows=false
      bMiteredCorners=false
      bSeparateFillAlpha=false
      outlineMaterial=null
      outlineSize=0
      outlineColor=<rgba 0 0 0 1>
     }
    }
   }
   slot {class=BorderSlot}
  }
 ]
}
