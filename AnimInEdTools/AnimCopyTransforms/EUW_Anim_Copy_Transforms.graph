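# Editor utility widget that copies world transforms between actors and Control Rig controls in the focused Level Sequence.
# Selections become source and destination list items carrying binding IDs, resolved back through Sequencer bindings.
# Copy samples the selected source's transform at the current frame and stores each destination's offset from it, with separate absolute and relative paste buttons.
# Tick updates button states, optionally auto-refreshing destinations; checkboxes choose current-frame, time-slider or specified ranges.
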EditorUtilityWidgetBlueprint /AnimInEdTools/AnimCopyTransforms/EUW_Anim_Copy_Transforms
graphs [
 Binding ID Equality {nodes [Binding ID Equality() {is=title K2Node_FunctionEntry}]}
 Clear Destinations {nodes [
   Clear Destinations() {is=title K2Node_FunctionEntry}
   > Clear(TargetArray:Array of BP Anim Copy List Item Data Soft Object References=Get Destinations() -> (Destinations:Array of BP Anim Copy List Item Data Soft Object References))
   > ClearListItems(self=Get CommonListView_Dest() -> (CommonListView_Dest:Common List View Object Reference))
   > Clear(TargetArray:Array of Transforms=Get Destination Relative Transforms() -> (Destination Relative Transforms:Array of Transforms))
  ]}
 Clear Sources {nodes [
   Clear Sources() {is=title K2Node_FunctionEntry}
   > Clear(TargetArray:Array of BP Anim Copy List Item Data Soft Object References=Get Sources() -> (Sources:Array of BP Anim Copy List Item Data Soft Object References))
   > ClearListItems(self=Get CommonListView_Source() -> (CommonListView_Source:Common List View Object Reference))
   > Set Selected Item Text(Selected Item Text:Text="LOCTEXT(\"---\")") -> (Output_Get:Text)
  ]}
 Color Source And Dest Buttons {nodes [
   Color Source And Dest Buttons() {is=title K2Node_FunctionEntry}
   > n1 = Sequence()
   n2 = Get Destinations()
   n3 = Branch(execute=@n1.then_1, Condition="integer > integer"(A=Length(TargetArray:Array of BP Anim Copy List Item Data Soft Object References=@n2)))
   n4 = Get Button_Set_Dest() -> (Button_Set_Dest:Editor Utility Button Object Reference)
   > Set!(
    execute
    BackgroundColor_R:float=0.45
    BackgroundColor_G:float=1.8
    BackgroundColor_B:float=2.7
    BackgroundColor_A:float=1
    self:Button Object Reference=@n4
   ) -> (then, Output_Get:Linear Color Structure)
   > Set Enable Clear Destinations(Enable Clear Destinations:bool=true) -> (Output_Get:bool)
   Set(execute=@n3.else, BackgroundColor_R=0.85, BackgroundColor_G=0.85, BackgroundColor_B=0.85, BackgroundColor_A=1, self=@n4)
   > Set Enable Clear Destinations(Enable Clear Destinations=false)
   n5 = Get Sources()
   n6 = Branch(execute=@n1.then_0, Condition="integer > integer"(A=Length(TargetArray=@n5)))
   n7 = Get Button_Set_Source() -> (Button_Set_Source:Editor Utility Button Object Reference)
   > Set(BackgroundColor_R=0.45, BackgroundColor_G=1.8, BackgroundColor_B=2.7, BackgroundColor_A=1, self=@n7)
   > Set Enable Clear Sources(Enable Clear Sources:bool=true) -> (Output_Get:bool)
   Set(execute=@n6.else, BackgroundColor_R=0.85, BackgroundColor_G=0.85, BackgroundColor_B=0.85, BackgroundColor_A=1, self=@n7)
   > Set Enable Clear Sources(Enable Clear Sources=false)
  ]}
 Copy Transforms {nodes [
   Copy Transforms() {is=title K2Node_FunctionEntry}
   > Clear(TargetArray:Array of Transforms=Get Destination Relative Transforms())
   > n1 = Cast To BP_AnimCopy_ListItem_data!(
    execute
    Object=Get Selected Item!(self:List View Object Reference=Get CommonListView_Source()) -> (ReturnValue:Object Reference) {is=title K2Node_CallFunction}
   ) -> (then, CastFailed, AsBP Anim Copy List Item Data:BP Anim Copy List Item Data Object Reference) {is=title K2Node_DynamicCast}
   n2 = Make Soft Reference!(Input:BP Anim Copy List Item Data Object Reference=@n1) -> (Output:BP Anim Copy List Item Data Soft Object Reference)
   > Set Selected BP Item(Selected BP Item:BP Anim Copy List Item Data Soft Object Reference=@n2) -> (Output_Get:BP Anim Copy List Item Data Soft Object Reference)
   > n3 = Set Copied BP Item(Copied BP Item:BP Anim Copy List Item Data Soft Object Reference=@n2) -> (Output_Get:BP Anim Copy List Item Data Soft Object Reference)
   > n4 = Get Current Local Time!(execute, TimeUnit:EMovieSceneTimeUnit Enum=DisplayRate) -> (
    then
    ReturnValue_Frame_FrameNumber_Value:int
    ReturnValue_Frame_SubFrame:float=~
    ReturnValue_Time:float=~
    ReturnValue_MarkedFrame:String
    ReturnValue_Timecode:Timecode Structure
    ReturnValue_PositionType:EMovieScenePositionType Enum
    ReturnValue_UpdateMethod:EUpdatePositionMethod Enum
    ReturnValue_bHasJumped:bool=~
   ) {is=title K2Node_CallFunction}
   n5 = Make Frame Number!(Value:int=@n4) -> (FrameNumber:Frame Number Structure) {is=title K2Node_MakeStruct}
   > n6 = Get World Transform From Binding ID!(execute, self, In BP AnimCopy:BP Anim Copy List Item Data Soft Object Reference=@n3, Current Frame:Frame Number Structure=@n5) -> (then, Out Transform:Transform) {is=title K2Node_CallFunction}
   > n7 = Set Copied Transform(Copied Transform:Transform=@n6) -> (Output_Get:Transform)
   > Set Selected Transform(Selected Transform:Transform=@n7) -> (Output_Get:Transform)
   > Set Copied Item Text(Copied Item Text:Text=Get Input Text!(self:BP Anim Copy List Item Data Object Reference=Resolve Soft Reference!(Input:BP Anim Copy List Item Data Soft Object Reference=@n3) -> (Output:BP Anim Copy List Item Data Object Reference)) -> (Input Text:Text)) -> (Output_Get:Text)
   > n8 = For Each Loop(Array:Array of BP Anim Copy List Item Data Soft Object References=Get Destinations()) -> (Array Element:BP Anim Copy List Item Data Soft Object Reference)
   > n9 = Get World Transform From Binding ID(In BP AnimCopy=@n8, Current Frame=@n5)
   > Add(
    TargetArray:Array of Transforms=Get Destination Relative Transforms()
    NewItem:"Transform (by ref)"=MakeRelativeTransform!(A:"Transform (by ref)"=@n9, RelativeTo:"Transform (by ref)"=Get Copied Transform() -> (Copied Transform:Transform)) -> (ReturnValue:Transform)
   ) {is=title K2Node_CallArrayFunction}
   Set Current Transform(Current Transform:Transform=ComposeTransforms!(
     A:"Transform (by ref)"=Get Current Transform() -> (Current Transform:Transform)
     B:"Transform (by ref)"=Get Parent Transform() -> (Parent Transform:Transform)
    ) -> (ReturnValue:Transform)) -> (Output_Get:Transform)
  ]}
 Enable Copy {nodes [
   Enable Copy() {is=title K2Node_FunctionEntry}
   n1 = Length(TargetArray=Get Sources())
   n2 = Length(TargetArray=Get Destinations())
   > n3 = Branch(Condition=AND Boolean(A="integer > integer"(A=@n1), B="integer > integer"(A=@n2)))
   > Set Ready To Copy(Ready To Copy:bool=true) -> (Output_Get:bool)
   Set Ready To Copy(execute=@n3.else, Ready To Copy=false)
   > Set Ready To Paste(Ready To Paste:bool=false) -> (Output_Get:bool)
  ]}
 EventGraph {nodes [
   Delay!(execute, WorldContextObject:Object Reference, Duration=0.2) -> (then)
   n1 = Get Sources()
   > n2 = Branch(Condition="integer > integer"(A=Length(TargetArray=@n1), B=1))
   > Set Visibility(Visibility=Visible)
   Set Visibility(execute=@n2.else, Visibility=Hidden)
   "On Clicked (Button_Set_Dest)"() {is=title K2Node_ComponentBoundEvent}
   > Set Destinations() {is=title K2Node_CallFunction}
   n3 = "On Check State Changed (CurrentFrameCheckBox)"!() -> (OutputDelegate, then, bIsChecked:bool) {is=title K2Node_ComponentBoundEvent}
   > n4 = Branch(Condition=@n3.bIsChecked)
   > SetCheckedState!(
    execute
    self:Check Box Object Reference=Get TimeSliderCheckBox() -> (TimeSliderCheckBox:Editor Utility Check Box Object Reference)
    InCheckedState:ECheckBoxState Enum=Unchecked
   ) -> (then)
   > n5 = SetCheckedState(self=Get SpecifiedRangeCheckBox() -> (SpecifiedRangeCheckBox:Editor Utility Check Box Object Reference), InCheckedState=Unchecked)
   Set Visibility(
    execute=[@n8.then, @n5.then]
    Visibility=Hidden
    self=Get FrameRangeHorizontalBox() -> (FrameRangeHorizontalBox:Horizontal Box Object Reference)
   )
   n6 = IsChecked(self=Get SpecifiedRangeCheckBox())
   n7 = IsChecked(self=Get TimeSliderCheckBox())
   Branch(
    execute=@n4.else
    Condition=AND Boolean(
     A="Equal (Boolean)"(A:bool=@n7, B:bool) -> (ReturnValue:bool=~) {is=title K2Node_PromotableOperator}
     B="Equal (Boolean)"(A=@n6) -> (ReturnValue=~)
    )
   )
   > n8 = SetCheckedState(self=Get CurrentFrameCheckBox() -> (CurrentFrameCheckBox:Editor Utility Check Box Object Reference), InCheckedState=Checked)
   n9 = "On Check State Changed (SpecifiedRangeCheckBox)"!() -> (OutputDelegate, then, bIsChecked:bool) {is=title K2Node_ComponentBoundEvent}
   > n10 = Branch(Condition=@n9.bIsChecked)
   > SetCheckedState(self=Get CurrentFrameCheckBox(), InCheckedState=Unchecked)
   > n11 = SetCheckedState(self=Get TimeSliderCheckBox(), InCheckedState=Unchecked)
   Set Visibility(execute=[@n15.then, @n11.then], Visibility=Visible, self=Get FrameRangeHorizontalBox())
   > n12 = GetPlaybackRange!(
    execute
    Sequence:Movie Scene Sequence Object Reference=Get Focused Level Sequence() -> (Focused Level Sequence:Level Sequence Object Reference)
   ) -> (then, ReturnValue_bHasStart:bool=~, ReturnValue_bHasEnd:bool=~, ReturnValue_InclusiveStart:int, ReturnValue_ExclusiveEnd:int)
   > Set Text(
    Text="To Text (Integer)"(Value=@n12.ReturnValue_InclusiveStart)
    self:Text Box Object Reference=Get StartFrameTextBox() -> (StartFrameTextBox:Editor Utility Editable Text Box Object Reference)
   )
   > Set Text(
    Text="To Text (Integer)"(Value=@n12.ReturnValue_ExclusiveEnd)
    self=Get EndFrameTextBox() -> (EndFrameTextBox:Editor Utility Editable Text Box Object Reference)
   )
   n13 = IsChecked(self=Get CurrentFrameCheckBox())
   n14 = IsChecked(self=Get TimeSliderCheckBox())
   Branch(execute=@n10.else, Condition=AND Boolean(A="Equal (Boolean)"(A=@n13) -> (ReturnValue=~), B="Equal (Boolean)"(A=@n14) -> (ReturnValue=~)))
   > n15 = SetCheckedState(self=Get SpecifiedRangeCheckBox(), InCheckedState=Checked)
   "On Clicked (Button_Paste_Transforms_Rel)"() {is=title K2Node_ComponentBoundEvent}
   > n16 = Set PasteRelative_BOOL(PasteRelative_BOOL:bool=true) -> (Output_Get:bool)
   Paste Transforms(execute=[@n19.then, @n16.then])
   "On Clicked (StartFrame_BTN)"() {is=title K2Node_ComponentBoundEvent}
   > n17 = Get Current Local Time(TimeUnit=DisplayRate) -> (ReturnValue_Frame_SubFrame=~, ReturnValue_Time=~, ReturnValue_bHasJumped=~)
   > Set Text(Text="To Text (Integer)"(Value=@n17), self=Get StartFrameTextBox())
   "On Clicked (EndFrame_BTN)"() {is=title K2Node_ComponentBoundEvent}
   > n18 = Get Current Local Time(TimeUnit=DisplayRate) -> (ReturnValue_Frame_SubFrame=~, ReturnValue_Time=~, ReturnValue_bHasJumped=~)
   > Set Text(Text="To Text (Integer)"(Value=@n18), self=Get EndFrameTextBox())
   "On Clicked (Button_Clear_Dest)"() {is=title K2Node_ComponentBoundEvent}
   > Clear Destinations() {is=title K2Node_CallFunction}
   > Set Ready To Paste(Ready To Paste=false)
   "On Clicked (Button_Paste_Transforms)"() {is=title K2Node_ComponentBoundEvent}
   > n19 = Set PasteRelative_BOOL(PasteRelative_BOOL=false)
   "On Clicked (Button_Set_Source)"() {is=title K2Node_ComponentBoundEvent}
   > Set Sources() {is=title K2Node_CallFunction}
   "On Clicked (Button_Copy_Transforms)"() {is=title K2Node_ComponentBoundEvent}
   > Copy Transforms() {is=title K2Node_CallFunction}
   > Set Ready To Paste(Ready To Paste=true)
   n20 = "On Item Selection Changed (CommonListView_Source)"!() -> (OutputDelegate, then, Item:Object Reference, bIsSelected:bool) {is=title K2Node_ComponentBoundEvent}
   > n21 = Cast To BP_AnimCopy_ListItem_data(Object=@n20.Item)
   > Set Selected Item Text(Selected Item Text=Get Input Text(self=@n21))
   > n22 = Cast To BP_AnimCopy_ListItem_data(Object=Get Selected Item(self=Get CommonListView_Source()))
   > Set Selected BP Item(Selected BP Item=Make Soft Reference(Input=@n22))
   "On Clicked (Button_Clear_Source)"() {is=title K2Node_ComponentBoundEvent}
   > Clear Sources() {is=title K2Node_CallFunction}
   > Set Ready To Copy(Ready To Copy=false)
   n23 = "On Check State Changed (AutoRefreshCheckBox)"!() -> (OutputDelegate, then, bIsChecked:bool) {is=title K2Node_ComponentBoundEvent}
   > Set Auto Refresh Dest(Auto Refresh Dest:bool=@n23.bIsChecked) -> (Output_Get:bool)
   n24 = "On Check State Changed (TimeSliderCheckBox)"!() -> (OutputDelegate, then, bIsChecked:bool) {is=title K2Node_ComponentBoundEvent}
   > n25 = Branch(Condition=@n24.bIsChecked)
   > SetCheckedState(self=Get CurrentFrameCheckBox(), InCheckedState=Unchecked)
   > n26 = SetCheckedState(self=Get SpecifiedRangeCheckBox(), InCheckedState=Unchecked)
   Set Visibility(execute=[@n29.then, @n26.then], Visibility=Hidden, self=Get FrameRangeHorizontalBox())
   n27 = IsChecked(self=Get SpecifiedRangeCheckBox())
   n28 = IsChecked(self=Get CurrentFrameCheckBox())
   Branch(execute=@n25.else, Condition=AND Boolean(A="Equal (Boolean)"(A=@n28) -> (ReturnValue=~), B="Equal (Boolean)"(A=@n27) -> (ReturnValue=~)))
   > n29 = SetCheckedState(self=Get TimeSliderCheckBox(), InCheckedState=Checked)
   Event Tick()
   > n30 = Sequence!(execute) -> (then_0, then_1, then_2:Exec, then_3:Exec)
   n31 = Branch(execute=@n30.then_2, Condition=Get Ready To Paste() -> (Ready To Paste:bool))
   > Set Display Copy Item Text(Display Copy Item Text:Text=Get Copied Item Text() -> (Copied Item Text:Text)) -> (Output_Get:Text)
   Set Display Copy Item Text(execute=@n31.else, Display Copy Item Text="LOCTEXT(\"---\")")
   Enable Copy(execute=@n30.then_1) {is=title K2Node_CallFunction}
   > Color Source And Dest Buttons() {is=title K2Node_CallFunction}
   > Branch(Condition=Get Auto Refresh Dest() -> (Auto Refresh Dest:bool))
   > Set Destinations()
   n32 = GetFocusedLevelSequence!(execute=@n30.then_0) -> (then, ReturnValue:Level Sequence Object Reference)
   > n33 = Set Focused Level Sequence(Focused Level Sequence:Level Sequence Object Reference=@n32) -> (Output_Get:Level Sequence Object Reference)
   > Is Valid(InputObject=@n33)
   > Set Frame Range() {is=title K2Node_CallFunction}
   n34 = Branch(execute=@n30.then_3, Condition=Get Ready To Paste())
   > Set Display Paste Item Text(Display Paste Item Text:Text=Get Selected Item Text() -> (Selected Item Text:Text)) -> (Output_Get:Text)
   Set Display Paste Item Text(execute=@n34.else, Display Paste Item Text="LOCTEXT(\"---\")")
   "On Hovered (Button_Set_Source)"() {is=title K2Node_ComponentBoundEvent}
   Event PreConstruct()
   Event Construct()
  ]}
 Get Actor From Control Rig {nodes [
   n1 = Get Actor From Control Rig!() -> (then, In Control Rig:Control Rig Soft Object Reference) {is=title K2Node_FunctionEntry}
   > n2 = Get CR Track From Control Rig!(execute, self, In Control Rig:Control Rig Soft Object Reference=@n1) -> (then, Out CR Track:Movie Scene Control Rig Parameter Track Soft Object Reference) {is=title K2Node_CallFunction}
   > Set Compare CR Track(Compare CR Track:Movie Scene Control Rig Parameter Track Object Reference=Resolve Soft Reference(Input:Movie Scene Control Rig Parameter Track Soft Object Reference=@n2) -> (Output:Movie Scene Control Rig Parameter Track Object Reference)) -> (Output_Get:Movie Scene Control Rig Parameter Track Object Reference)
   > n3 = GetFocusedLevelSequence()
   > n4 = Set Current Level Sequence(Current Level Sequence:Level Sequence Object Reference=@n3) -> (Output_Get:Level Sequence Object Reference)
   > n5 = GetBindings!(execute, Sequence:Movie Scene Sequence Object Reference=@n4) -> (then, ReturnValue:Array of Movie Scene Binding Proxy Structures)
   > n6 = For Each Loop(Array:Array of Movie Scene Binding Proxy Structures=@n5) -> (Array Element:Movie Scene Binding Proxy Structure)
   Return Node!(
    execute=@n6.Completed
    Out Soft Actor:Actor Soft Object Reference=Make Soft Reference(Input:Actor Object Reference=Get Return Actor() -> (Return Actor:Actor Object Reference)) -> (Output:Actor Soft Object Reference)
   )
   > n7 = GetBindingID!(
    execute
    Sequence:Movie Scene Sequence Object Reference=Get Current Level Sequence() -> (Current Level Sequence:Level Sequence Object Reference)
    InBinding:"Movie Scene Binding Proxy Structure (by ref)"=@n6
   ) -> (then, ReturnValue:Movie Scene Object Binding ID Structure)
   > n8 = For Each Loop(Array:Array of Object References=GetBoundObjects!(ObjectBinding:Movie Scene Object Binding ID Structure=@n7) -> (ReturnValue:Array of Object References)) -> (Array Element:Object Reference)
   > n9 = Cast To Actor!(execute, Object=@n8) -> (then, CastFailed, AsActor:Actor Object Reference) {is=title K2Node_DynamicCast}
   > n10 = GetTracks!(execute, InBinding:"Movie Scene Binding Proxy Structure (by ref)"=@n6) -> (then, ReturnValue:Array of Movie Scene Track Object References)
   > n11 = For Each Loop(Array:Array of Movie Scene Track Object References=@n10) -> (Array Element:Movie Scene Track Object Reference)
   > n12 = Cast To MovieSceneControlRigParameterTrack!(execute, Object=@n11) -> (then, CastFailed, AsMovie Scene Control Rig Parameter Track:Movie Scene Control Rig Parameter Track Object Reference) {is=title K2Node_DynamicCast}
   > Branch(Condition="Equal (Object)"(A=@n12, B=Get Compare CR Track() -> (Compare CR Track:Movie Scene Control Rig Parameter Track Object Reference)))
   > Set Return Actor(Return Actor:Actor Object Reference=@n9) -> (Output_Get:Actor Object Reference)
  ]}
 Get Actor from Binding ID {nodes [
   Get Actor from Binding ID!() -> (then, In Movie Scene Binding ID:Movie Scene Object Binding ID Structure) {is=title K2Node_FunctionEntry}
   > n1 = GetFocusedLevelSequence()
   > Set Current Level Sequence(Current Level Sequence=@n1)
   > n2 = Sequence!(execute) -> (then_0, then_1)
   n3 = For Each Loop(
    Exec=@n2.then_1
    Array:Array of Movie Scene Binding Proxy Structures=Get Bindings() -> (Bindings:Array of Movie Scene Binding Proxy Structures)
   ) -> (Array Element:Movie Scene Binding Proxy Structure)
   Return Node(execute=@n3.Completed, Out Soft Actor=Get Return Soft Actor() -> (Return Soft Actor:Actor Soft Object Reference))
   > n4 = GetBindingID(Sequence=Get Current Level Sequence(), InBinding=@n3)
   > n5 = ResolveBindingID!(
    execute
    RootSequence:Movie Scene Sequence Object Reference=Get Current Level Sequence()
    InObjectBindingID:Movie Scene Object Binding ID Structure=@n4
   ) -> (then, ReturnValue:Movie Scene Binding Proxy Structure)
   > n6 = GetDisplayName!(execute:Exec, InBinding:"Movie Scene Binding Proxy Structure (by ref)"=@n5) -> (then:Exec, ReturnValue:Text)
   > n7 = ResolveBindingID(
    RootSequence=Get Current Level Sequence()
    InObjectBindingID=Get In Movie Scene Binding ID() -> (In Movie Scene Binding ID:Movie Scene Object Binding ID Structure)
   )
   > n8 = GetDisplayName(InBinding=@n7)
   > Branch(Condition="Equal Exactly (Text)"!(A:"Text (by ref)"=@n8, B:"Text (by ref)"=@n6) -> (ReturnValue:bool))
   > n9 = For Each Loop(Array:Array of Object References=GetBoundObjects(ObjectBinding=@n4)) -> (Array Element:Object Reference)
   > n10 = Cast To Actor(Object=@n9)
   > Set Return Soft Actor(Return Soft Actor:Actor Soft Object Reference=Make Soft Reference(Input=@n10)) -> (Output_Get:Actor Soft Object Reference)
   n11 = GetBindings(execute=@n2.then_0, Sequence=Get Current Level Sequence())
   > Set Bindings(Bindings:Array of Movie Scene Binding Proxy Structures=@n11) -> (Output_Get:Array of Movie Scene Binding Proxy Structures)
   "Equal (Guid)"(
    A:"Guid Structure (by ref)"=ResolveBindingID!(execute, RootSequence, InObjectBindingID=Get In Movie Scene Binding ID()) -> (then, ReturnValue_BindingID:Guid Structure, ReturnValue_Sequence:Movie Scene Sequence Object Reference)
    B:"Guid Structure (by ref)"=Break Movie Scene Binding Proxy!(MovieSceneBindingProxy:"Movie Scene Binding Proxy Structure (by ref)"=@n3) -> (BindingID:Guid Structure, Sequence:Movie Scene Sequence Object Reference) {is=title K2Node_BreakStruct}
   ) -> (ReturnValue:bool=~) {is=title K2Node_PromotableOperator}
  ]}
 Get All Actors {nodes [
   Get All Actors() {is=title K2Node_FunctionEntry}
   > Clear(TargetArray:Array of Actor Soft Object References=Get Return All Actors() -> (Return All Actors:Array of Actor Soft Object References))
   > Clear(TargetArray:Array of Movie Scene Object Binding ID Structures=Get Return All Binding IDs() -> (Return All Binding IDs:Array of Movie Scene Object Binding ID Structures))
   > n1 = GetFocusedLevelSequence()
   > n2 = Set Current Level Sequence(Current Level Sequence=@n1)
   > n3 = GetBindings(Sequence=@n2)
   > n4 = For Each Loop(Array:Array of Movie Scene Binding Proxy Structures=@n3) -> (Array Element:Movie Scene Binding Proxy Structure)
   Return Node!(
    execute=@n4.Completed
    Out All Soft Actors:Array of Actor Soft Object References=Get Return All Actors()
    Out Actor Binding IDs:Array of Movie Scene Object Binding ID Structures=Get Return All Binding IDs()
   )
   > n5 = GetBindingID(InBinding=@n4)
   > n6 = For Each Loop(Array:Array of Object References=GetBoundObjects(ObjectBinding=@n5)) -> (Array Element:Object Reference)
   > n7 = Cast To Actor(Object=@n6)
   > Add(
    TargetArray:Array of Actor Soft Object References=Get Return All Actors()
    NewItem:"Actor Soft Object Reference (by ref)"=Make Soft Reference(Input=@n7)
   )
   > Add(
    TargetArray:Array of Movie Scene Object Binding ID Structures=Get Return All Binding IDs()
    NewItem:"Movie Scene Object Binding ID Structure (by ref)"=@n5
   )
  ]}
 Get Binding ID From Control Rig {nodes [
   Get Binding ID From Control Rig!() -> (then, In Control Rig:Control Rig Soft Object Reference) {is=title K2Node_FunctionEntry}
   > n1 = GetFocusedLevelSequence()
   > n2 = GetControlRigs!(execute, LevelSequence:Level Sequence Object Reference=@n1) -> (then, ReturnValue:Array of Control Rig Sequencer Binding Proxy Structures)
   > n3 = For Each Loop(Array:Array of Control Rig Sequencer Binding Proxy Structures=@n2) -> (Array Element:Control Rig Sequencer Binding Proxy Structure)
   n4 = Break Control Rig Sequencer Binding Proxy!(ControlRigSequencerBindingProxy:"Control Rig Sequencer Binding Proxy Structure (by ref)"=@n3) -> (
    Proxy:Movie Scene Binding Proxy Structure
    ControlRig:Control Rig Object Reference
    Track:Movie Scene Control Rig Parameter Track Object Reference
   ) {is=title K2Node_BreakStruct}
   > n5 = GetBindingID(InBinding=@n4)
   n6 = Get In Control Rig() -> (In Control Rig:Control Rig Soft Object Reference)
   > Branch(Condition="Equal (Object)"(A=@n4.ControlRig, B=Resolve Soft Reference(Input:Control Rig Soft Object Reference=@n6) -> (Output:Control Rig Object Reference)))
   > Set Return Binding ID(Return Binding ID:Movie Scene Object Binding ID Structure=@n5) -> (Output_Get:Movie Scene Object Binding ID Structure)
   Return Node!(
    execute=@n3.Completed
    Out Binding ID:Movie Scene Object Binding ID Structure=Get Return Binding ID() -> (Return Binding ID:Movie Scene Object Binding ID Structure)
   )
  ]}
 Get Binding ID from Actor {nodes [
   n1 = Get Binding ID from Actor!() -> (then, In Soft Actor:Actor Soft Object Reference) {is=title K2Node_FunctionEntry}
   > Set Hard Actor(Hard Actor:Actor Object Reference=Resolve Soft Reference(Input:Actor Soft Object Reference=@n1) -> (Output:Actor Object Reference)) -> (Output_Get:Actor Object Reference)
   > n2 = GetFocusedLevelSequence()
   > Set Current Level Sequence(Current Level Sequence=@n2)
   > n3 = Sequence()
   n4 = For Each Loop(Exec=@n3.then_1, Array:Array of Movie Scene Binding Proxy Structures=Get Bindings()) -> (Array Element:Movie Scene Binding Proxy Structure)
   Return Node!(
    execute=@n4.Completed
    Out Movie Scene Binding ID:Movie Scene Object Binding ID Structure=Get Return Movie Binding ID() -> (Return Movie Binding ID:Movie Scene Object Binding ID Structure)
   )
   > n5 = GetBindingID(Sequence=Get Current Level Sequence(), InBinding=@n4)
   > n6 = For Each Loop(Array:Array of Object References=GetBoundObjects(ObjectBinding=@n5)) -> (Array Element:Object Reference)
   > n7 = Cast To Actor(Object=@n6)
   > Branch(Condition="Equal (Object)"(A=@n7, B=Get Hard Actor() -> (Hard Actor:Actor Object Reference)))
   > Set Return Movie Binding ID(Return Movie Binding ID:Movie Scene Object Binding ID Structure=@n5) -> (Output_Get:Movie Scene Object Binding ID Structure)
   n8 = GetBindings(execute=@n3.then_0, Sequence=Get Current Level Sequence())
   > Set Bindings(Bindings=@n8)
  ]}
 Get CR Track From Control Rig {nodes [
   Get CR Track From Control Rig!() -> (then, In Control Rig:Control Rig Soft Object Reference) {is=title K2Node_FunctionEntry}
   > n1 = GetFocusedLevelSequence()
   > n2 = Set Current Level Sequence(Current Level Sequence=@n1)
   > n3 = GetControlRigs(LevelSequence=@n2)
   > n4 = For Each Loop(Array:Array of Control Rig Sequencer Binding Proxy Structures=@n3) -> (Array Element:Control Rig Sequencer Binding Proxy Structure)
   Return Node!(
    execute=@n4.Completed
    Out CR Track:Movie Scene Control Rig Parameter Track Soft Object Reference=Make Soft Reference(Input:Movie Scene Control Rig Parameter Track Object Reference=Get Return CR Track() -> (Return CR Track:Movie Scene Control Rig Parameter Track Object Reference)) -> (Output:Movie Scene Control Rig Parameter Track Soft Object Reference)
   )
   n5 = Break Control Rig Sequencer Binding Proxy(ControlRigSequencerBindingProxy=@n4)
   n6 = Get In Control Rig()
   > Branch(Condition="Equal (Object)"(A=@n5.ControlRig, B=Resolve Soft Reference(Input:Control Rig Soft Object Reference=@n6) -> (Output:Control Rig Object Reference)))
   > Set Return CR Track(Return CR Track:Movie Scene Control Rig Parameter Track Object Reference=@n5.Track) -> (Output_Get:Movie Scene Control Rig Parameter Track Object Reference)
  ]}
 Get CR Track From Control Rig Binding ID {nodes [
   n1 = Get CR Track From Control Rig Binding ID!() -> (then, In Movie Scene Binding ID:Movie Scene Object Binding ID Structure) {is=title K2Node_FunctionEntry}
   > n2 = GetFocusedLevelSequence()
   > n3 = Set Current Level Sequence(Current Level Sequence=@n2)
   > n4 = ResolveBindingID!(execute, RootSequence=@n3, InObjectBindingID=@n1) -> (then, ReturnValue:Movie Scene Binding Proxy Structure)
   > n5 = GetTracks(InBinding=@n4)
   > n6 = For Each Loop(Array:Array of Movie Scene Track Object References=@n5) -> (Array Element:Movie Scene Track Object Reference)
   Return Node!(
    execute=@n6.Completed
    Our Transform Track:Movie Scene Control Rig Parameter Track Soft Object Reference=Make Soft Reference(Input=Get Return Control Rig Track() -> (Return Control Rig Track:Movie Scene Control Rig Parameter Track Object Reference))
    Found CR Track:bool=Get NewLocalVar() -> (NewLocalVar:bool)
   )
   > n7 = Cast To MovieSceneControlRigParameterTrack(Object=@n6)
   > Set Return Control Rig Track(Return Control Rig Track:Movie Scene Control Rig Parameter Track Object Reference=@n7) -> (Output_Get:Movie Scene Control Rig Parameter Track Object Reference)
   > Set NewLocalVar(NewLocalVar:bool=true) -> (Output_Get:bool)
  ]}
 Get Control Rig From Binding ID {nodes [
   Get Control Rig From Binding ID!() -> (then, In Binding ID:Movie Scene Object Binding ID Structure) {is=title K2Node_FunctionEntry}
   > n1 = GetFocusedLevelSequence()
   > n2 = Set Current Level Sequence(Current Level Sequence=@n1)
   > n3 = GetControlRigs(LevelSequence=@n2)
   > n4 = For Each Loop(Array:Array of Control Rig Sequencer Binding Proxy Structures=@n3) -> (Array Element:Control Rig Sequencer Binding Proxy Structure)
   n5 = Break Control Rig Sequencer Binding Proxy(ControlRigSequencerBindingProxy=@n4)
   > n6 = GetDisplayName(InBinding=@n5)
   > n7 = ResolveBindingID(RootSequence=Get Current Level Sequence(), InObjectBindingID=Get In Binding ID() -> (In Binding ID:Movie Scene Object Binding ID Structure))
   > n8 = GetDisplayName(InBinding=@n7)
   > Branch(Condition="Equal Exactly (Text)"(A=@n8, B=@n6))
   > Set Return Control Rig(Return Control Rig:Control Rig Object Reference=@n5.ControlRig) -> (Output_Get:Control Rig Object Reference)
   Return Node!(
    execute=@n4.Completed
    Out Control Rig:Control Rig Soft Object Reference=Make Soft Reference(Input:Control Rig Object Reference=Get Return Control Rig() -> (Return Control Rig:Control Rig Object Reference)) -> (Output:Control Rig Soft Object Reference)
   )
   "Equal (Guid)"(A=Break Movie Scene Binding Proxy(), B=Break Movie Scene Binding Proxy()) -> (ReturnValue=~)
  ]}
 Get Control Rig From CR Track {nodes [
   Get Control Rig From CR Track!() -> (then, In CR Track:Movie Scene Control Rig Parameter Track Soft Object Reference) {is=title K2Node_FunctionEntry}
   > n1 = GetFocusedLevelSequence()
   > n2 = Set Current Level Sequence(Current Level Sequence=@n1)
   > n3 = GetControlRigs(LevelSequence=@n2)
   > n4 = For Each Loop(Array=@n3)
   Return Node(execute=@n4.Completed, Out Control Rig=Make Soft Reference(Input=Get Return Control Rig()))
   n5 = Break Control Rig Sequencer Binding Proxy(ControlRigSequencerBindingProxy=@n4)
   n6 = Get In CR Track() -> (In CR Track:Movie Scene Control Rig Parameter Track Soft Object Reference)
   > Branch(Condition="Equal (Object)"(
     A=@n5.Track
     B=Resolve Soft Reference(Input:Movie Scene Control Rig Parameter Track Soft Object Reference=@n6) -> (Output:Movie Scene Control Rig Parameter Track Object Reference)
    ))
   > Set Return Control Rig(Return Control Rig=@n5.ControlRig)
  ]}
 Get Selected Actors {nodes [
   Get Selected Actors() {is=title K2Node_FunctionEntry}
   > Clear(TargetArray:Array of Actor Soft Object References=Get Selected Soft Actors() -> (Selected Soft Actors:Array of Actor Soft Object References))
   > Clear(TargetArray:Array of Movie Scene Object Binding ID Structures=Get Selected Out IDs() -> (Selected Out IDs:Array of Movie Scene Object Binding ID Structures))
   > n1 = GetSelectedActors()
   > n2 = For Each Loop(Array:Array of Actor Object References=@n1) -> (Array Element:Actor Object Reference)
   Return Node!(
    execute=@n2.Completed
    Out Selected Soft Actors:Array of Actor Soft Object References=Get Selected Soft Actors()
    Out Binding IDs:Array of Movie Scene Object Binding ID Structures
   )
   n3 = Make Soft Reference(Input:Actor Object Reference=@n2) -> (Output:Actor Soft Object Reference)
   > Add(TargetArray:Array of Actor Soft Object References=Get Selected Soft Actors(), NewItem:"Actor Soft Object Reference (by ref)"=@n3)
   > n4 = Get Binding ID from Actor!(execute, self, In Soft Actor:Actor Soft Object Reference=@n3) -> (then, Out Movie Scene Binding ID:Movie Scene Object Binding ID Structure) {is=title K2Node_CallFunction}
   > Add(
    TargetArray:Array of Movie Scene Object Binding ID Structures=Get Selected Out IDs()
    NewItem:"Movie Scene Object Binding ID Structure (by ref)"=@n4
   )
  ]}
 Get Selected Actors And Control Rigs {nodes [
   Get Selected Actors And Control Rigs() {is=title K2Node_FunctionEntry}
   > Clear(TargetArray:Array of BP Anim Copy List Item Data Soft Object References=Get Return Out Objects() -> (Return Out Objects:Array of BP Anim Copy List Item Data Soft Object References))
   > n1 = Sequence!(execute) -> (then_0, then_1, then_2:Exec)
   > n2 = Get Selected Rigs And Controls!(execute, self) -> (
    then
    Selected Control Rigs:Array of Control Rig Soft Object References
    Selected Binding IDs:Array of Movie Scene Object Binding ID Structures
    Selected Control Names:Array of Names
    Selected Rig Names:Array of Strings
    Selected Tracks:Array of Movie Scene Control Rig Parameter Track Soft Object References
   ) {is=title K2Node_CallFunction}
   > n3 = For Each Loop(Array:Array of Control Rig Soft Object References=@n2) -> (Array Element:Control Rig Soft Object Reference)
   > n4 = Get Actor From Control Rig!(execute, self, In Control Rig:Control Rig Soft Object Reference=@n3) -> (then, Out Soft Actor:Actor Soft Object Reference) {is=title K2Node_CallFunction}
   > n5 = Get CR Track From Control Rig(In Control Rig=@n3) {is=title K2Node_CallFunction}
   > n6 = GetDisplayName!(execute, Track:Movie Scene Track Object Reference=Resolve Soft Reference(Input=@n5)) -> (then, ReturnValue)
   n7 = "Get (a copy)"(Array:Array of Names=@n2.Selected Control Names, Dimension 1=@n3.Array Index) -> (Output:Name)
   > n8 = Execute Python Script!(execute, PythonScript="ot_control_name = in_control_name.split(':')[-1]", in_control_name:String="To String (Name)"(InName=@n7)) -> (then, ReturnValue, ot_control_name:String) {inputs [in_control_name], outputs [ot_control_name]}
   n9 = GetActorLabel(self=Resolve Soft Reference(Input:Actor Soft Object Reference=@n4) -> (Output:Actor Object Reference), bCreateIfNone=true)
   n10 = "To String (Text)"(InText=@n6)
   > n11 = Construct Object from Class!(
    execute
    Class=&/BP_AnimCopy_ListItem_data.BP_AnimCopy_ListItem_data_C
    self
    Input Text:Text="To Text (String)"(InString=Append!(A=@n9, B=" : ", C:String=@n10, D:String=" : ", E:String=@n8.ot_control_name) -> (ReturnValue))
    Binding ID:Movie Scene Object Binding ID Structure="Get (a copy)"(Array:Array of Movie Scene Object Binding ID Structures=@n2.Selected Binding IDs, Dimension 1=@n3.Array Index) -> (Output:Movie Scene Object Binding ID Structure)
    Type:int=0
    Ctrl Name:Name=@n7
   ) -> (then, ReturnValue:BP Anim Copy List Item Data Object Reference)
   > Add(
    TargetArray:Array of BP Anim Copy List Item Data Soft Object References=Get Return Out Objects()
    NewItem:"BP Anim Copy List Item Data Soft Object Reference (by ref)"=Make Soft Reference(Input:BP Anim Copy List Item Data Object Reference=@n11) -> (Output:BP Anim Copy List Item Data Soft Object Reference)
   )
   n12 = Get Selected Actors!(execute=@n1.then_1, self) -> (then, Out Selected Soft Actors:Array of Actor Soft Object References, Out Binding IDs:Array of Movie Scene Object Binding ID Structures) {is=title K2Node_CallFunction}
   > n13 = For Each Loop(Array:Array of Actor Soft Object References=@n12) -> (Array Element:Actor Soft Object Reference)
   > n14 = Get Binding ID from Actor(In Soft Actor=@n13)
   n15 = GetActorLabel(self=Resolve Soft Reference(Input=@n13), bCreateIfNone=true)
   > n16 = Construct Object from Class(
    Class=&BP_AnimCopy_ListItem_data_C
    Input Text="To Text (String)"(InString=@n15)
    Binding ID=@n14
    Type=1
    Ctrl Name=String To Name(InString=@n15)
   )
   > Add(TargetArray=Get Return Out Objects(), NewItem=Make Soft Reference(Input=@n16))
   Return Node!(execute=@n1.then_2, Out Objects:Array of BP Anim Copy List Item Data Soft Object References=Get Return Out Objects())
  ]}
 Get Selected Rigs And Controls {nodes [
   Get Selected Rigs And Controls() {is=title K2Node_FunctionEntry}
   > Clear(TargetArray:Array of Control Rig Soft Object References=Get Return Control Rigs() -> (Return Control Rigs:Array of Control Rig Soft Object References))
   > Clear(TargetArray:Array of Movie Scene Object Binding ID Structures=Get Return Binding IDs() -> (Return Binding IDs:Array of Movie Scene Object Binding ID Structures))
   > Clear(TargetArray:Array of Names=Get Return Control Names() -> (Return Control Names:Array of Names))
   > Clear(TargetArray:Array of Strings=Get Return Rig Names() -> (Return Rig Names:Array of Strings))
   > Clear(TargetArray:Array of Movie Scene Control Rig Parameter Track Soft Object References=Get Return Rig Tracks() -> (Return Rig Tracks:Array of Movie Scene Control Rig Parameter Track Soft Object References))
   > n1 = GetFocusedLevelSequence()
   > n2 = Set Current Level Sequence(Current Level Sequence=@n1)
   > n3 = GetControlRigs(LevelSequence=@n2)
   > n4 = For Each Loop(Array:Array of Control Rig Sequencer Binding Proxy Structures=@n3) -> (Array Element:Control Rig Sequencer Binding Proxy Structure)
   n5 = Break Control Rig Sequencer Binding Proxy(ControlRigSequencerBindingProxy=@n4)
   > n6 = GetBindingID(Sequence=Get Current Level Sequence(), InBinding=@n5)
   > Set Current Binding ID(Current Binding ID:Movie Scene Object Binding ID Structure=@n6) -> (Output_Get:Movie Scene Object Binding ID Structure)
   > Set Current Control Rig(Current Control Rig:Control Rig Object Reference=@n5.ControlRig) -> (Output_Get:Control Rig Object Reference)
   > Set Current Rig Track(Current Rig Track:Movie Scene Control Rig Parameter Track Object Reference=@n5.Track) -> (Output_Get:Movie Scene Control Rig Parameter Track Object Reference)
   n7 = GetHierarchy!(self:Control Rig Object Reference=@n5.ControlRig) -> (ReturnValue:Rig Hierarchy Object Reference)
   > n8 = GetController!(execute, self:Rig Hierarchy Object Reference=@n7, bCreateIfNeeded:bool=true) -> (then, ReturnValue:Rig Hierarchy Controller Object Reference)
   > n9 = For Each Loop(Array:Array of Rig Element Key Structures=Get Controls!(self:Rig Hierarchy Object Reference=@n7, bTraverse:bool=true) -> (ReturnValue:Array of Rig Element Key Structures) {is=title K2Node_CallFunction}) -> (Array Element:Rig Element Key Structure)
   n10 = Get Current Control Rig() -> (Current Control Rig:Control Rig Object Reference)
   n11 = Break Rig Element Key!(RigElementKey:"Rig Element Key Structure (by ref)"=@n9) -> (Type:ERigElementType Enum, Name:Name) {is=title K2Node_BreakStruct}
   n12 = Break Rig Control Settings!(RigControlSettings:"Rig Control Settings Structure (by ref)"=GetControlSettings!(self:Rig Hierarchy Controller Object Reference=@n8, InKey:Rig Element Key Structure=@n9) -> (ReturnValue:Rig Control Settings Structure)) -> (
    AnimationType:ERigControlAnimationType Enum
    ControlType:ERigControlType Enum
    DisplayName:Name
    PrimaryAxis:ERigControlAxis Enum
    LimitEnabled:Array of Rig Control Limit Enabled Structures
    bDrawLimits:bool=~
    MinimumValue:Rig Control Value Structure
    MaximumValue:Rig Control Value Structure
    bShapeVisible:bool=~
    ShapeVisibility:ERigControlVisibility Enum
    ShapeName:Name
    ShapeColor:Linear Color Structure
    bIsTransientControl:bool=~
    ControlEnum:Enum Object Reference
    Customization:Rig Control Element Customization Structure
    DrivenControls:Array of Rig Element Key Structures
    bGroupWithParentControl:bool=~
    bRestrictSpaceSwitching:bool=~
    FilteredChannels:Array of ERigControlTransformChannel Enums
    PreferredRotationOrder:EEulerRotationOrder Enum
    bUsePreferredRotationOrder:bool=~
    ShapeTransform:Transform
   ) {is=title K2Node_BreakStruct}
   > Branch(Condition=AND Boolean!(
     A=IsControlSelected!(self:Control Rig Object Reference=@n10, InControlName:"Name (by ref)"=@n11.Name) -> (ReturnValue:bool)
     B="Equal (Enum)"(A:ERigControlAnimationType Enum=@n12, B:ERigControlAnimationType Enum=AnimationControl)
     C:bool="Equal (Enum)"(A:ERigControlType Enum=@n12.ControlType, B:ERigControlType Enum=EulerTransform)
    ) -> (ReturnValue))
   > Add(
    TargetArray:Array of Control Rig Soft Object References=Get Return Control Rigs()
    NewItem:"Control Rig Soft Object Reference (by ref)"=Make Soft Reference(Input:Control Rig Object Reference=Get Current Control Rig()) -> (Output:Control Rig Soft Object Reference)
   )
   > Add(TargetArray:Array of Names=Get Return Control Names(), NewItem:"Name (by ref)"=@n11.Name)
   > Add(
    TargetArray:Array of Strings=Get Return Rig Names()
    NewItem:"String (by ref)"=GetDisplayName!(Object:Object Reference=@n10) -> (ReturnValue:String)
   )
   > Add(
    TargetArray:Array of Movie Scene Object Binding ID Structures=Get Return Binding IDs()
    NewItem:"Movie Scene Object Binding ID Structure (by ref)"=Get Current Binding ID() -> (Current Binding ID:Movie Scene Object Binding ID Structure)
   )
   > Add(
    TargetArray:Array of Movie Scene Control Rig Parameter Track Soft Object References=Get Return Rig Tracks()
    NewItem:"Movie Scene Control Rig Parameter Track Soft Object Reference (by ref)"=Make Soft Reference(Input:Movie Scene Control Rig Parameter Track Object Reference=Get Current Rig Track() -> (Current Rig Track:Movie Scene Control Rig Parameter Track Object Reference)) -> (Output:Movie Scene Control Rig Parameter Track Soft Object Reference)
   )
   Return Node!(
    execute=@n4.Completed
    Selected Control Rigs:Array of Control Rig Soft Object References=Get Return Control Rigs()
    Selected Binding IDs:Array of Movie Scene Object Binding ID Structures=Get Return Binding IDs()
    Selected Control Names:Array of Names=Get Return Control Names()
    Selected Rig Names:Array of Strings=Get Return Rig Names()
    Selected Tracks:Array of Movie Scene Control Rig Parameter Track Soft Object References=Get Return Rig Tracks()
   )
  ]}
 Get Transform Track From Actor Binding ID {nodes [
   n1 = Get Transform Track From Actor Binding ID!() -> (then, In Movie Scene Binding ID:Movie Scene Object Binding ID Structure) {is=title K2Node_FunctionEntry}
   > n2 = GetFocusedLevelSequence()
   > n3 = Set Current Level Sequence(Current Level Sequence=@n2)
   > n4 = ResolveBindingID(RootSequence=@n3, InObjectBindingID=@n1)
   > n5 = GetTracks(InBinding=@n4)
   > n6 = For Each Loop(Array:Array of Movie Scene Track Object References=@n5) -> (Array Element:Movie Scene Track Object Reference)
   Return Node!(
    execute=@n6.Completed
    Our Transform Track:Movie Scene 3DTransform Track Soft Object Reference=Make Soft Reference(Input:Movie Scene 3DTransform Track Object Reference=Get Return Transform Track() -> (Return Transform Track:Movie Scene 3DTransform Track Object Reference)) -> (Output:Movie Scene 3DTransform Track Soft Object Reference)
   )
   > n7 = Cast To MovieScene3DTransformTrack!(execute, Object=@n6) -> (then, CastFailed, AsMovie Scene 3DTransform Track:Movie Scene 3DTransform Track Object Reference) {is=title K2Node_DynamicCast}
   > Set Return Transform Track(Return Transform Track:Movie Scene 3DTransform Track Object Reference=@n7) -> (Output_Get:Movie Scene 3DTransform Track Object Reference)
  ]}
 Get World Transform From Binding ID {nodes [
   Get World Transform From Binding ID!() -> (then, In BP AnimCopy:BP Anim Copy List Item Data Soft Object Reference, Current Frame:Frame Number Structure) {is=title K2Node_FunctionEntry}
   > n1 = Sequence!(execute) -> (then_0, then_1)
   > n2 = GetFocusedLevelSequence()
   > Set Current Level Sequence(Current Level Sequence=@n2)
   > n3 = Set BP Anim Copy(BP Anim Copy:BP Anim Copy List Item Data Object Reference=Resolve Soft Reference(Input:BP Anim Copy List Item Data Soft Object Reference=Get In BP AnimCopy() -> (In BP AnimCopy:BP Anim Copy List Item Data Soft Object Reference)) -> (Output:BP Anim Copy List Item Data Object Reference)) -> (Output_Get:BP Anim Copy List Item Data Object Reference)
   n4 = Get Type!(self:BP Anim Copy List Item Data Object Reference=@n3) -> (Type:int=0)
   > n5 = Branch(Condition="Equal (Integer)"(A=@n4))
   n6 = Get BP Anim Copy() -> (BP Anim Copy:BP Anim Copy List Item Data Object Reference)
   > n7 = Get Control Rig From Binding ID!(
    execute
    self
    In Binding ID:Movie Scene Object Binding ID Structure=Get Binding ID!(self:BP Anim Copy List Item Data Object Reference=@n6) -> (Binding ID:Movie Scene Object Binding ID Structure)
   ) -> (then, Out Control Rig:Control Rig Soft Object Reference) {is=title K2Node_CallFunction}
   > n8 = GetControlRigWorldTransform!(
    execute
    LevelSequence:Level Sequence Object Reference=Get Current Level Sequence()
    ControlRig:Control Rig Object Reference=Resolve Soft Reference(Input:Control Rig Soft Object Reference=@n7) -> (Output:Control Rig Object Reference)
    ControlName:Name=Get Ctrl Name!(self:BP Anim Copy List Item Data Object Reference=@n6) -> (Ctrl Name:Name=None)
    Frame:Frame Number Structure=Get Current Frame() -> (Current Frame:Frame Number Structure)
    TimeUnit:EMovieSceneTimeUnit Enum=DisplayRate
   ) -> (then, ReturnValue:Transform)
   > Set Return Transform(Return Transform:Transform=@n8) -> (Output_Get:Transform)
   Branch(execute=@n5.else, Condition="Equal (Integer)"(A=@n4, B=1))
   > n9 = Get Actor from Binding ID!(execute, self, In Movie Scene Binding ID:Movie Scene Object Binding ID Structure=Get Binding ID(self=Get BP Anim Copy())) -> (then, Out Soft Actor:Actor Soft Object Reference) {is=title K2Node_CallFunction}
   > n10 = GetActorWorldTransform!(
    execute
    LevelSequence:Level Sequence Object Reference=Get Current Level Sequence()
    Actor:Actor Object Reference=Resolve Soft Reference(Input:Actor Soft Object Reference=@n9) -> (Output:Actor Object Reference)
    Frame:Frame Number Structure=Get Current Frame()
    TimeUnit:EMovieSceneTimeUnit Enum=DisplayRate
   ) -> (then, ReturnValue:Transform)
   > Set Return Transform(Return Transform=@n10)
   Return Node!(execute=@n1.then_1, Out Transform:Transform=Get Return Transform() -> (Return Transform:Transform))
  ]}
 Paste Transforms {nodes [
   Paste Transforms() {is=title K2Node_FunctionEntry}
   > BeginTransaction()
   > n1 = Sequence()
   > n2 = For Each Loop(Array:Array of BP Anim Copy List Item Data Soft Object References=Get Destinations()) -> (Array Element:BP Anim Copy List Item Data Soft Object Reference)
   > Set Current Dest(Current Dest:BP Anim Copy List Item Data Soft Object Reference=@n2) -> (Output_Get:BP Anim Copy List Item Data Soft Object Reference)
   > Set Current Rel Dest Transform(Current Rel Dest Transform:Transform="Get (a copy)"(Array:Array of Transforms=Get Destination Relative Transforms(), Dimension 1=@n2.Array Index) -> (Output:Transform)) -> (Output_Get:Transform)
   n3 = Get!() -> (Playback Range_bHasStart:bool=~, Playback Range_bHasEnd:bool=~, Playback Range_InclusiveStart:int, Playback Range_ExclusiveEnd:int)
   > n4 = For Loop(FirstIndex=@n3.Playback Range_InclusiveStart, LastIndex=@n3.Playback Range_ExclusiveEnd)
   n5 = Make Frame Number(Value=@n4)
   > n6 = Get World Transform From Binding ID(In BP AnimCopy=Get Copied BP Item() -> (Copied BP Item:BP Anim Copy List Item Data Soft Object Reference), Current Frame=@n5) {is=title K2Node_CallFunction}
   > Set Transform To Paste(Transform To Paste:Transform=@n6) -> (Output_Get:Transform)
   > n7 = Get World Transform From Binding ID(In BP AnimCopy=Get Selected BP Item() -> (Selected BP Item:BP Anim Copy List Item Data Soft Object Reference), Current Frame=@n5)
   > Set Selected Rel Parent Transform(Selected Rel Parent Transform:Transform=@n7) -> (Output_Get:Transform)
   > For Loop(LastIndex=Get Paste Loops() -> (Paste Loops:int=0))
   > n8 = Get World Transform From Binding ID(In BP AnimCopy=Get Current Dest() -> (Current Dest:BP Anim Copy List Item Data Soft Object Reference), Current Frame=@n5)
   > n9 = Sequence()
   > Branch(Condition=Get PasteRelative_BOOL() -> (PasteRelative_BOOL:bool))
   > Set Transform To Paste(Transform To Paste=ComposeTransforms(
     A=Get Current Rel Dest Transform() -> (Current Rel Dest Transform:Transform)
     B=Get Selected Rel Parent Transform() -> (Selected Rel Parent Transform:Transform)
    ))
   n10 = vector - vector(A=BreakTransform(InTransform=Get Transform To Paste() -> (Transform To Paste:Transform)), B=BreakTransform(InTransform=@n8))
   Branch(execute=@n9.then_1, Condition="float > float"(A=Vector Length(A=@n10), B=Get Paste Tolerance() -> (Paste Tolerance:double)))
   > Set World Transform on Bound Obj to ID!(
    execute
    self
    In BP AnimCopy:BP Anim Copy List Item Data Soft Object Reference=Get Current Dest()
    Current Frame:Frame Number Structure=@n5
    In Transform:Transform=Get Transform To Paste()
   ) -> (then) {is=title K2Node_CallFunction}
   EndTransaction(execute=@n1.then_1)
  ]}
 Set Destinations {nodes [
   Set Destinations() {is=title K2Node_FunctionEntry}
   > Clear Destinations()
   > n1 = Get Selected Actors And Control Rigs!(execute, self) -> (then, Out Objects:Array of BP Anim Copy List Item Data Soft Object References) {is=title K2Node_CallFunction}
   > n2 = For Each Loop(Array=@n1)
   > AddItem!(
    execute
    self:List View Object Reference=Get CommonListView_Dest()
    Item:Object Reference=Resolve Soft Reference(Input:BP Anim Copy List Item Data Soft Object Reference=@n2) -> (Output:BP Anim Copy List Item Data Object Reference)
   ) -> (then)
   > Add(
    TargetArray:Array of BP Anim Copy List Item Data Soft Object References=Get Destinations()
    NewItem:"BP Anim Copy List Item Data Soft Object Reference (by ref)"=@n2
   )
  ]}
 Set Frame Range {nodes [
   Set Frame Range() {is=title K2Node_FunctionEntry}
   > n1 = GetFocusedLevelSequence()
   > Set Focused Level Sequence(Focused Level Sequence=@n1)
   > n2 = Sequence()
   > n3 = Branch(Condition=IsChecked(self=Get CurrentFrameCheckBox()))
   > n4 = Get Current Local Time(TimeUnit=DisplayRate) -> (ReturnValue_Frame_SubFrame=~, ReturnValue_Time=~, ReturnValue_bHasJumped=~)
   > Set!(
    execute
    Playback Range_bHasStart:bool=true
    Playback Range_bHasEnd:bool=true
    Playback Range_InclusiveStart:int=@n4
    Playback Range_ExclusiveEnd:int=@n4
   ) -> (then, Output_Get:Sequencer Scripting Range Structure)
   n5 = Branch(execute=@n3.else, Condition=IsChecked(self=Get SpecifiedRangeCheckBox()))
   n6 = Get Text(self=Get EndFrameTextBox())
   n7 = Get Text(self=Get StartFrameTextBox())
   > Set(
    Playback Range_bHasStart=true
    Playback Range_bHasEnd=true
    Playback Range_InclusiveStart=String To Integer(InString="To String (Text)"(InText=@n7))
    Playback Range_ExclusiveEnd=String To Integer(InString="To String (Text)"(InText=@n6))
   )
   n8 = GetPlaybackRange!(execute=@n5.else, Sequence=Get Focused Level Sequence()) -> (then, ReturnValue:Sequencer Scripting Range Structure)
   > Set Playback Range(Playback Range:Sequencer Scripting Range Structure=@n8) -> (Output_Get:Sequencer Scripting Range Structure)
   n9 = Get() -> (Playback Range_bHasStart=~, Playback Range_bHasEnd=~)
   Set Int Range Start(execute=@n2.then_1, Int Range Start:int=@n9.Playback Range_InclusiveStart) -> (Output_Get:int=0)
   > Set Int Range End(Int Range End:int=@n9.Playback Range_ExclusiveEnd) -> (Output_Get:int=0)
  ]}
 Set Sources {nodes [
   Set Sources() {is=title K2Node_FunctionEntry}
   > Clear Sources()
   > n1 = Get Selected Actors And Control Rigs()
   > n2 = For Each Loop(Array=@n1)
   n3 = Get CommonListView_Source()
   Clear Selection(execute=@n2.Completed, self:List View Object Reference=@n3)
   > SetSelectedIndex!(execute, self:List View Object Reference=@n3, Index:int=0) -> (then)
   AddItem(execute=@n2.LoopBody, self=Get CommonListView_Source(), Item=Resolve Soft Reference(Input=@n2))
   > Add(TargetArray=Get Sources(), NewItem=@n2)
  ]}
 Set World Transform on Bound Obj to ID {nodes [
   Set World Transform on Bound Obj to ID!() -> (then, In BP AnimCopy:BP Anim Copy List Item Data Soft Object Reference, Current Frame:Frame Number Structure, In Transform:Transform) {is=title K2Node_FunctionEntry}
   > n1 = GetFocusedLevelSequence()
   > Set Current Level Sequence(Current Level Sequence=@n1)
   > n2 = Set BP Anim Copy(BP Anim Copy=Resolve Soft Reference(Input=Get In BP AnimCopy()))
   n3 = Get Type(self=@n2) -> (Type=0)
   > n4 = Branch(Condition="Equal (Integer)"(A=@n3))
   n5 = Get BP Anim Copy()
   > n6 = Get Control Rig From Binding ID(In Binding ID=Get Binding ID(self=@n5))
   > SetControlRigWorldTransform!(
    execute
    LevelSequence:Level Sequence Object Reference=Get Current Level Sequence()
    ControlRig:Control Rig Object Reference=Resolve Soft Reference(Input:Control Rig Soft Object Reference=@n6) -> (Output:Control Rig Object Reference)
    ControlName:Name=Get Ctrl Name(self=@n5) -> (Ctrl Name=None)
    Frame:Frame Number Structure=Get Current Frame()
    WorldTransform:"Transform (by ref)"=Get In Transform() -> (In Transform:Transform)
    TimeUnit:EMovieSceneTimeUnit Enum=DisplayRate
    bSetKey:bool=true
   ) -> (then)
   Branch(execute=@n4.else, Condition="Equal (Integer)"(A=@n3, B=1))
   n7 = Get Binding ID(self=Get BP Anim Copy())
   > n8 = Get Actor from Binding ID(In Movie Scene Binding ID=@n7)
   > Set Current Actor(Current Actor:Actor Object Reference=Resolve Soft Reference(Input:Actor Soft Object Reference=@n8) -> (Output:Actor Object Reference)) -> (Output_Get:Actor Object Reference)
   > n9 = ResolveBindingID(RootSequence=Get Current Level Sequence(), InObjectBindingID=@n7)
   > n10 = GetTracks(InBinding=@n9)
   > n11 = For Each Loop(Array:Array of Movie Scene Track Object References=@n10) -> (Array Element:Movie Scene Track Object Reference)
   > n12 = Cast To MovieScene3DTransformTrack(Object=@n11)
   > Execute Python Script!(
    execute
    PythonScript="import unreal\r\n\r\ndef setKeyOnActorAtFrame(level_sequence, actor, transform_track, frame_number, global_transform):    \r\n    \r\n    if isinstance(transform_track, unreal.MovieScene3DTransformTrack):\r\n        track_sections = transform_track.get_sections()\r\n    \r\n        if transform_track and track_sections:\r\n            track_section = track_sections[0]\r\n            \r\n            track_channels = track_section.get_all_channels()       \r\n        \r\n            local_translation = global_transform.translation\r\n        \r\n            local_rotation = global_transform.rotation.rotator()\r\n            \r\n            actorParent = actor.get_attach_parent_actor()\r\n            \r\n            if actorParent != None:\r\n                local_translation = actorParent.get_actor_transform().inverse_transform_location(global_transform.translation)\r\n                local_rotation = actorParent.get_actor_transform().inverse_transform_rotation(global_transform.rotation.rotator())\r\n        \r\n            for channel in track_channels:\r\n                if 'Location.X' in channel.get_name():\r\n                    channel.add_key(frame_number, local_translation.x)\r\n                if 'Location.Y' in channel.get_name():\r\n                    channel.add_key(frame_number, local_translation.y)\r\n                if 'Location.Z' in channel.get_name():\r\n                    channel.add_key(frame_number, local_translation.z)\r\n                if 'Rotation.X' in channel.get_name():\r\n                    channel.add_key(frame_number, local_rotation.roll)\r\n                if 'Rotation.Y' in channel.get_name():\r\n                    channel.add_key(frame_number, local_rotation.pitch)\r\n                if 'Rotation.Z' in channel.get_name():\r\n                    channel.add_key(frame_number, local_rotation.yaw) \r\n\r\nsetKeyOnActorAtFrame(level_sequence, actor, transform_track, frame_number, global_transform)\r\nprint('TRIED TO COPY TO AN ACTOR')"
    level_sequence:Level Sequence Object Reference=Get Current Level Sequence()
    actor:Actor Object Reference=Get Current Actor() -> (Current Actor:Actor Object Reference)
    frame_number:Frame Number Structure=Get Current Frame()
    global_transform:Transform=Get In Transform()
    transform_track:Movie Scene 3DTransform Track Object Reference=@n12
   ) -> (then, ReturnValue) {inputs [level_sequence, actor, frame_number, global_transform, transform_track]}
  ]}
]
variables [
 Auto Refresh Dest {type=boolean}
 Button Set Source {type=EditorUtilityButton, default=null}
 Copied BP Item {type=BP_AnimCopy_ListItem_data_C, default=null}
 Copied Item Text {type=string, default=""}
 Copied Transform {type=Transform, default {location=<0 0 0>, rotation=<rot 0 0 0>, scale=<1 1 1>}}
 Destination Relative Transforms {type="array<Transform>", default []}
 Destinations {type="array<BP_AnimCopy_ListItem_data_C>", default []}
 Display Copy Item Text {type=string, default=""}
 Display Paste Item Text {type=string, default=""}
 Enable Clear Destinations {type=boolean}
 Enable Clear Sources {type=boolean}
 Focused Level Sequence {type=LevelSequence, default=null}
 Frame Number {type=FrameNumber, default {value=0}}
 Frame Range End {type=FrameNumber, default {value=0}}
 Frame Range Start {type=FrameNumber, default {value=0}}
 Int Range End {type=integer}
 Int Range Start {type=integer}
 Paste Loops {type=integer, default=15}
 Paste Tolerance {type=number, default=0.01}
 PasteRelative_BOOL {type=boolean}
 Playback Range {
  type=SequencerScriptingRange
  default {bHasEnd=false, bHasStart=false, exclusiveEnd=0, inclusiveStart=0, internalRate {denominator=1, numerator=60000}}
 }
 Ready To Copy {type=boolean}
 Ready To Paste {type=boolean}
 Selected BP Item {type=BP_AnimCopy_ListItem_data_C, default=null}
 Selected Item Text {type=string, default=---}
 Selected Transform {type=Transform, default {location=<0 0 0>, rotation=<rot 0 0 0>, scale=<1 1 1>}}
 Sources {type="array<BP_AnimCopy_ListItem_data_C>", default []}
 startFrame {type=integer}
 Use Frame Range {type=boolean}
]
widgets {
 info {rootWidgetClass=&CanvasPanel, widgetCount=60}
 widgets [
  CanvasPanel_0 {widgetClassPath=&CanvasPanel, widget=CanvasPanel_0}
  VerticalBox_266 {
   widgetClassPath=&VerticalBox
   parent=CanvasPanel_0
   widget=VerticalBox_266
   slot {
    class=CanvasPanelSlot
    properties {bAutoSize=true, layoutData {alignment=<0 0>, anchors {maximum=<1 0>, minimum=<0 0>}, offsets=<ltrb 0 0 0 4296.6665>}}
   }
  }
  ScaleBox_159 {widgetClassPath=&ScaleBox, parent=VerticalBox_266, widget=ScaleBox_159, slot {class=VerticalBoxSlot}}
  VerticalBox_42 {
   widgetClassPath=&VerticalBox
   parent=ScaleBox_159
   widget=VerticalBox_42
   slot {class=ScaleBoxSlot, properties {horizontalAlignment=HAlign_Fill, verticalAlignment=VAlign_Fill}}
  }
  loadButtons_Horiz {
   widgetClassPath=&HorizontalBox
   parent=VerticalBox_42
   widget=loadButtons_Horiz
   slot {class=VerticalBoxSlot, properties {verticalAlignment=VAlign_Top}}
  }
  LoadChild_horiz {
   widgetClassPath=&HorizontalBox
   parent=loadButtons_Horiz
   widget=LoadChild_horiz
   slot {class=HorizontalBoxSlot, properties {padding=<ltrb 0 0 0 10>, size {sizeRule=Fill, value=1}}}
  }
  Spacer_4 {
   widgetClassPath=&Spacer
   parent=LoadChild_horiz
   widget=Spacer_4
   properties {size=<50 1>}
   slot {class=HorizontalBoxSlot, properties {size {sizeRule=Fill, value=0.1}}}
  }
  Button_Set_Dest {
   widgetClassPath=&EditorUtilityButton
   parent=LoadChild_horiz
   widget=Button_Set_Dest
   properties {
    clipping=ClipToBounds
    widgetStyle {
     clickedSlateSound {resourceObject=null}
     disabled {
      bIsDynamicallyLoaded=false
      drawAs=RoundedBox
      imageType=NoImage
      mirroring=NoMirror
      resourceName=None
      resourceObject=null
      tiling=NoTile
      imageSize=<0 0>
      margin=<ltrb 0 0 0 0>
      outlineSettings {
       bUseBrushTransparency=false
       roundingType=FixedRadius
       width=1
       color {colorUseRule=UseColor_ColorTable, specifiedColor=<rgba 0 0 0 0>}
       cornerRadii=<4 4 4 4>
      }
      tintColor {colorUseRule=UseColor_Specified, specifiedColor=<rgba 0.0395 0.0395 0.0395 1>}
      uVRegion {bIsValid=false, max=<0 0>, min=<0 0>}
     }
     disabledForeground {colorUseRule=UseColor_ColorTable, specifiedColor=<rgba 0 0 0 0>}
     hovered {
      bIsDynamicallyLoaded=false
      drawAs=RoundedBox
      imageType=NoImage
      mirroring=NoMirror
      resourceName=None
      resourceObject=null
      tiling=NoTile
      imageSize=<32 32>
      margin=<ltrb 0 0 0 0>
      outlineSettings {
       bUseBrushTransparency=false
       roundingType=HalfHeightRadius
       width=1
       color {colorUseRule=UseColor_ColorTable, specifiedColor=<rgba 0 0 0 0>}
       cornerRadii=<4 4 4 4>
      }
      tintColor {colorUseRule=UseColor_Specified, specifiedColor=<rgba 0.25 0.25 0.25 1>}
      uVRegion {bIsValid=false, max=<0 0>, min=<0 0>}
     }
     hoveredForeground {colorUseRule=UseColor_ColorTable, specifiedColor=<rgba 0 0 0 0>}
     hoveredSlateSound {resourceObject=null}
     normal {
      bIsDynamicallyLoaded=false
      drawAs=RoundedBox
      imageType=NoImage
      mirroring=NoMirror
      resourceName=None
      resourceObject=null
      tiling=NoTile
      imageSize=<32 32>
      margin=<ltrb 0 0 0 0>
      outlineSettings {
       bUseBrushTransparency=false
       roundingType=HalfHeightRadius
       width=1
       color {colorUseRule=UseColor_ColorTable, specifiedColor=<rgba 0 0 0 0>}
       cornerRadii=<4 4 4 4>
      }
      tintColor {colorUseRule=UseColor_Specified, specifiedColor=<rgba 0.085 0.085 0.085 1>}
      uVRegion {bIsValid=false, max=<0 0>, min=<0 0>}
     }
     normalForeground {colorUseRule=UseColor_ColorTable, specifiedColor=<rgba 0 0 0 0>}
     normalPadding=<ltrb 12 1.5 12 1.5>
     pressed {
      bIsDynamicallyLoaded=false
      drawAs=RoundedBox
      imageType=NoImage
      mirroring=NoMirror
      resourceName=None
      resourceObject=null
      tiling=NoTile
      imageSize=<32 32>
      margin=<ltrb 0 0 0 0>
      outlineSettings {
       bUseBrushTransparency=false
       roundingType=FixedRadius
       width=1
       color {colorUseRule=UseColor_ColorTable, specifiedColor=<rgba 0 0 0 0>}
       cornerRadii=<4 4 4 4>
      }
      tintColor {colorUseRule=UseColor_Specified, specifiedColor=<rgba 0.0284 0.0284 0.0284 1>}
      uVRegion {bIsValid=false, max=<0 0>, min=<0 0>}
     }
     pressedForeground {colorUseRule=UseColor_ColorTable, specifiedColor=<rgba 0 0 0 0>}
     pressedPadding=<ltrb 12 2.5 12 0.5>
     pressedSlateSound {resourceObject=null}
    }
   }
   slot {class=HorizontalBoxSlot, properties {size {sizeRule=Fill, value=1}}}
  }
  TextBlock_Get_Relatives {
   widgetClassPath=&TextBlock
   parent=Button_Set_Dest
   widget=TextBlock_Get_Relatives
   properties {
    justification=Center
    text="Load CHILDREN "
    font {
     bForceMonospaced=false
     bMaterialIsStencil=false
     fontMaterial=null
     fontObject=&Roboto
     letterSpacing=0
     monospacedWidth=1
     size=26.25
     skewAmount=0
     typefaceFontName=Bold
     outlineSettings {
      bApplyOutlineToDropShadows=false
      bMiteredCorners=false
      bSeparateFillAlpha=false
      outlineMaterial=null
      outlineSize=0
      outlineColor=<rgba 0 0 0 1>
     }
    }
   }
   slot {class=ButtonSlot, properties {horizontalAlignment=HAlign_Fill, verticalAlignment=VAlign_Fill}}
  }
  AutoRefreshCheckBox {
   widgetClassPath=&EditorUtilityCheckBox
   bIsVariable=true
   parent=LoadChild_horiz
   widget=AutoRefreshCheckBox
   slot {class=HorizontalBoxSlot, properties {horizontalAlignment=HAlign_Left, padding=<ltrb 5 0 5 0>}}
  }
  CurrentFrame_TextBlock_1 {
   widgetClassPath=&TextBlock
   parent=AutoRefreshCheckBox
   widget=CurrentFrame_TextBlock_1
   properties {
    text=" Auto"
    font {
     bForceMonospaced=false
     bMaterialIsStencil=false
     fontMaterial=null
     fontObject=&Roboto
     letterSpacing=0
     monospacedWidth=1
     size=22.5
     skewAmount=0
     typefaceFontName=Regular
     outlineSettings {
      bApplyOutlineToDropShadows=false
      bMiteredCorners=false
      bSeparateFillAlpha=false
      outlineMaterial=null
      outlineSize=0
      outlineColor=<rgba 0 0 0 1>
     }
    }
   }
   slot {class=PanelSlot}
  }
  Spacer_564 {
   widgetClassPath=&Spacer
   parent=LoadChild_horiz
   widget=Spacer_564
   properties {size=<50 1>}
   slot {class=HorizontalBoxSlot, properties {size {sizeRule=Fill, value=0.1}}}
  }
  LoadParent_horiz {
   widgetClassPath=&HorizontalBox
   parent=loadButtons_Horiz
   widget=LoadParent_horiz
   slot {class=HorizontalBoxSlot, properties {padding=<ltrb 5 0 5 10>, size {sizeRule=Fill, value=1}}}
  }
  Button_Set_Source {
   widgetClassPath=&EditorUtilityButton
   parent=LoadParent_horiz
   widget=Button_Set_Source
   properties {
    clipping=ClipToBounds
    widgetStyle {
     clickedSlateSound {resourceObject=null}
     disabled {
      bIsDynamicallyLoaded=false
      drawAs=RoundedBox
      imageType=NoImage
      mirroring=NoMirror
      resourceName=None
      resourceObject=null
      tiling=NoTile
      imageSize=<0 0>
      margin=<ltrb 0 0 0 0>
      outlineSettings {
       bUseBrushTransparency=false
       roundingType=FixedRadius
       width=1
       color {colorUseRule=UseColor_ColorTable, specifiedColor=<rgba 0 0 0 0>}
       cornerRadii=<4 4 4 4>
      }
      tintColor {colorUseRule=UseColor_Specified, specifiedColor=<rgba 0.0395 0.0395 0.0395 1>}
      uVRegion {bIsValid=false, max=<0 0>, min=<0 0>}
     }
     disabledForeground {colorUseRule=UseColor_ColorTable, specifiedColor=<rgba 0 0 0 0>}
     hovered {
      bIsDynamicallyLoaded=false
      drawAs=RoundedBox
      imageType=NoImage
      mirroring=NoMirror
      resourceName=None
      resourceObject=null
      tiling=NoTile
      imageSize=<32 32>
      margin=<ltrb 0 0 0 0>
      outlineSettings {
       bUseBrushTransparency=false
       roundingType=HalfHeightRadius
       width=1
       color {colorUseRule=UseColor_ColorTable, specifiedColor=<rgba 0 0 0 0>}
       cornerRadii=<4 4 4 4>
      }
      tintColor {colorUseRule=UseColor_Specified, specifiedColor=<rgba 0.25 0.25 0.25 1>}
      uVRegion {bIsValid=false, max=<0 0>, min=<0 0>}
     }
     hoveredForeground {colorUseRule=UseColor_ColorTable, specifiedColor=<rgba 0 0 0 0>}
     hoveredSlateSound {resourceObject=null}
     normal {
      bIsDynamicallyLoaded=false
      drawAs=RoundedBox
      imageType=NoImage
      mirroring=NoMirror
      resourceName=None
      resourceObject=null
      tiling=NoTile
      imageSize=<32 32>
      margin=<ltrb 0 0 0 0>
      outlineSettings {
       bUseBrushTransparency=false
       roundingType=HalfHeightRadius
       width=1
       color {colorUseRule=UseColor_ColorTable, specifiedColor=<rgba 0 0 0 0>}
       cornerRadii=<4 4 4 4>
      }
      tintColor {colorUseRule=UseColor_Specified, specifiedColor=<rgba 0.085 0.085 0.085 1>}
      uVRegion {bIsValid=false, max=<0 0>, min=<0 0>}
     }
     normalForeground {colorUseRule=UseColor_ColorTable, specifiedColor=<rgba 0 0 0 0>}
     normalPadding=<ltrb 12 1.5 12 1.5>
     pressed {
      bIsDynamicallyLoaded=false
      drawAs=RoundedBox
      imageType=NoImage
      mirroring=NoMirror
      resourceName=None
      resourceObject=null
      tiling=NoTile
      imageSize=<32 32>
      margin=<ltrb 0 0 0 0>
      outlineSettings {
       bUseBrushTransparency=false
       roundingType=FixedRadius
       width=1
       color {colorUseRule=UseColor_ColorTable, specifiedColor=<rgba 0 0 0 0>}
       cornerRadii=<4 4 4 4>
      }
      tintColor {colorUseRule=UseColor_Specified, specifiedColor=<rgba 0.0284 0.0284 0.0284 1>}
      uVRegion {bIsValid=false, max=<0 0>, min=<0 0>}
     }
     pressedForeground {colorUseRule=UseColor_ColorTable, specifiedColor=<rgba 0 0 0 0>}
     pressedPadding=<ltrb 12 2.5 12 0.5>
     pressedSlateSound {resourceObject=null}
    }
   }
   slot {class=HorizontalBoxSlot, properties {size {sizeRule=Fill, value=1}}}
  }
  TextBlock_Copy_Source {
   widgetClassPath=&TextBlock
   parent=Button_Set_Source
   widget=TextBlock_Copy_Source
   properties {
    justification=Center
    text="Load PARENT(s)"
    font {
     bForceMonospaced=false
     bMaterialIsStencil=false
     fontMaterial=null
     fontObject=&Roboto
     letterSpacing=0
     monospacedWidth=1
     size=26.25
     skewAmount=0
     typefaceFontName=Bold
     outlineSettings {
      bApplyOutlineToDropShadows=false
      bMiteredCorners=false
      bSeparateFillAlpha=false
      outlineMaterial=null
      outlineSize=0
      outlineColor=<rgba 0 0 0 1>
     }
    }
   }
   slot {class=ButtonSlot, properties {horizontalAlignment=HAlign_Fill, verticalAlignment=VAlign_Fill}}
  }
  lists_horiz {widgetClassPath=&HorizontalBox, parent=VerticalBox_42, widget=lists_horiz, slot {class=VerticalBoxSlot}}
  SizeBox_69 {
   widgetClassPath=&SizeBox
   parent=lists_horiz
   widget=SizeBox_69
   properties {bOverride_HeightOverride=true, heightOverride=100}
   slot {class=HorizontalBoxSlot, properties {padding=<ltrb 5 5 5 5>, size {sizeRule=Fill, value=1}}}
  }
  HorizontalBox_388 {widgetClassPath=&HorizontalBox, parent=SizeBox_69, widget=HorizontalBox_388, slot {class=SizeBoxSlot}}
  EditorUtilityScrollBox {
   widgetClassPath=&EditorUtilityScrollBox
   bIsVariable=true
   parent=HorizontalBox_388
   widget=EditorUtilityScrollBox
   properties {alwaysShowScrollbar=true, scrollbarThickness=<9 1>}
   slot {class=HorizontalBoxSlot, properties {padding=<ltrb 0 0 5 0>, size {sizeRule=Fill, value=1}}}
  }
  ChildList_ScaleBox {
   widgetClassPath=&ScaleBox
   parent=EditorUtilityScrollBox
   widget=ChildList_ScaleBox
   properties {stretch=Fill}
   slot {class=ScrollBoxSlot, properties {padding=<ltrb 10 0 5 0>, size {sizeRule=Fill, value=1}}}
  }
  Border_172 {
   widgetClassPath=&Border
   parent=ChildList_ScaleBox
   widget=Border_172
   properties {
    background {
     bIsDynamicallyLoaded=false
     drawAs=Image
     imageType=NoImage
     mirroring=NoMirror
     resourceName=None
     resourceObject=null
     tiling=NoTile
     imageSize=<32 32>
     margin=<ltrb 0 0 0 0>
     outlineSettings {
      bUseBrushTransparency=false
      roundingType=HalfHeightRadius
      width=0
      color {colorUseRule=UseColor_Specified, specifiedColor=<rgba 0 0 0 0>}
      cornerRadii=<0 0 0 0>
     }
     tintColor {colorUseRule=UseColor_Specified, specifiedColor=<rgba 0.0085 0.0085 0.0085 1>}
     uVRegion {bIsValid=false, max=<0 0>, min=<0 0>}
    }
    padding=<ltrb 0 0 0 0>
   }
   slot {class=ScaleBoxSlot, properties {horizontalAlignment=HAlign_Fill, verticalAlignment=VAlign_Fill}}
  }
  CommonListView_Dest {
   widgetClassPath=&CommonListView
   bIsVariable=true
   parent=Border_172
   widget=CommonListView_Dest
   properties {
    entryWidgetClass=&/Anim_Copy_Transform_Selection.Anim_Copy_Transform_Selection_C
    selectionMode=None
    scrollBarStyle {
     thickness=8
     draggedThumbImage {
      bIsDynamicallyLoaded=false
      drawAs=RoundedBox
      imageType=NoImage
      mirroring=NoMirror
      resourceName=None
      resourceObject=null
      tiling=NoTile
      imageSize=<0 0>
      margin=<ltrb 0 0 0 0>
      outlineSettings {
       bUseBrushTransparency=false
       roundingType=FixedRadius
       width=0
       color {colorUseRule=UseColor_Specified, specifiedColor=<rgba 0 0 0 0>}
       cornerRadii=<4 4 4 4>
      }
      tintColor {colorUseRule=UseColor_Specified, specifiedColor=<rgba 0.2159 0.2159 0.2159 1>}
      uVRegion {bIsValid=false, max=<0 0>, min=<0 0>}
     }
     horizontalBackgroundImage {
      bIsDynamicallyLoaded=false
      drawAs=NoDrawType
      imageType=NoImage
      mirroring=NoMirror
      resourceName=None
      resourceObject=null
      tiling=NoTile
      imageSize=<0 0>
      margin=<ltrb 0 0 0 0>
      outlineSettings {
       bUseBrushTransparency=false
       roundingType=HalfHeightRadius
       width=0
       color {colorUseRule=UseColor_Specified, specifiedColor=<rgba 0 0 0 0>}
       cornerRadii=<0 0 0 0>
      }
      tintColor {colorUseRule=UseColor_Specified, specifiedColor=<rgba 1 1 1 1>}
      uVRegion {bIsValid=false, max=<0 0>, min=<0 0>}
     }
     horizontalBottomSlotImage {
      bIsDynamicallyLoaded=false
      drawAs=NoDrawType
      imageType=NoImage
      mirroring=NoMirror
      resourceName=None
      resourceObject=null
      tiling=NoTile
      imageSize=<0 0>
      margin=<ltrb 0 0 0 0>
      outlineSettings {
       bUseBrushTransparency=false
       roundingType=HalfHeightRadius
       width=0
       color {colorUseRule=UseColor_Specified, specifiedColor=<rgba 0 0 0 0>}
       cornerRadii=<0 0 0 0>
      }
      tintColor {colorUseRule=UseColor_Specified, specifiedColor=<rgba 1 1 1 1>}
      uVRegion {bIsValid=false, max=<0 0>, min=<0 0>}
     }
     horizontalTopSlotImage {
      bIsDynamicallyLoaded=false
      drawAs=NoDrawType
      imageType=NoImage
      mirroring=NoMirror
      resourceName=None
      resourceObject=null
      tiling=NoTile
      imageSize=<0 0>
      margin=<ltrb 0 0 0 0>
      outlineSettings {
       bUseBrushTransparency=false
       roundingType=HalfHeightRadius
       width=0
       color {colorUseRule=UseColor_Specified, specifiedColor=<rgba 0 0 0 0>}
       cornerRadii=<0 0 0 0>
      }
      tintColor {colorUseRule=UseColor_Specified, specifiedColor=<rgba 1 1 1 1>}
      uVRegion {bIsValid=false, max=<0 0>, min=<0 0>}
     }
     hoveredThumbImage {
      bIsDynamicallyLoaded=false
      drawAs=RoundedBox
      imageType=NoImage
      mirroring=NoMirror
      resourceName=None
      resourceObject=null
      tiling=NoTile
      imageSize=<0 0>
      margin=<ltrb 0 0 0 0>
      outlineSettings {
       bUseBrushTransparency=false
       roundingType=FixedRadius
       width=0
       color {colorUseRule=UseColor_Specified, specifiedColor=<rgba 0 0 0 0>}
       cornerRadii=<4 4 4 4>
      }
      tintColor {colorUseRule=UseColor_Specified, specifiedColor=<rgba 0.2159 0.2159 0.2159 1>}
      uVRegion {bIsValid=false, max=<0 0>, min=<0 0>}
     }
     normalThumbImage {
      bIsDynamicallyLoaded=false
      drawAs=RoundedBox
      imageType=NoImage
      mirroring=NoMirror
      resourceName=None
      resourceObject=null
      tiling=NoTile
      imageSize=<0 0>
      margin=<ltrb 0 0 0 0>
      outlineSettings {
       bUseBrushTransparency=false
       roundingType=FixedRadius
       width=0
       color {colorUseRule=UseColor_Specified, specifiedColor=<rgba 0 0 0 0>}
       cornerRadii=<4 4 4 4>
      }
      tintColor {colorUseRule=UseColor_Specified, specifiedColor=<rgba 0.0953 0.0953 0.0953 1>}
      uVRegion {bIsValid=false, max=<0 0>, min=<0 0>}
     }
     verticalBackgroundImage {
      bIsDynamicallyLoaded=false
      drawAs=NoDrawType
      imageType=NoImage
      mirroring=NoMirror
      resourceName=None
      resourceObject=null
      tiling=NoTile
      imageSize=<0 0>
      margin=<ltrb 0 0 0 0>
      outlineSettings {
       bUseBrushTransparency=false
       roundingType=HalfHeightRadius
       width=0
       color {colorUseRule=UseColor_Specified, specifiedColor=<rgba 0 0 0 0>}
       cornerRadii=<0 0 0 0>
      }
      tintColor {colorUseRule=UseColor_Specified, specifiedColor=<rgba 1 1 1 1>}
      uVRegion {bIsValid=false, max=<0 0>, min=<0 0>}
     }
     verticalBottomSlotImage {
      bIsDynamicallyLoaded=false
      drawAs=NoDrawType
      imageType=NoImage
      mirroring=NoMirror
      resourceName=None
      resourceObject=null
      tiling=NoTile
      imageSize=<0 0>
      margin=<ltrb 0 0 0 0>
      outlineSettings {
       bUseBrushTransparency=false
       roundingType=HalfHeightRadius
       width=0
       color {colorUseRule=UseColor_Specified, specifiedColor=<rgba 0 0 0 0>}
       cornerRadii=<0 0 0 0>
      }
      tintColor {colorUseRule=UseColor_Specified, specifiedColor=<rgba 1 1 1 1>}
      uVRegion {bIsValid=false, max=<0 0>, min=<0 0>}
     }
     verticalTopSlotImage {
      bIsDynamicallyLoaded=false
      drawAs=NoDrawType
      imageType=NoImage
      mirroring=NoMirror
      resourceName=None
      resourceObject=null
      tiling=NoTile
      imageSize=<0 0>
      margin=<ltrb 0 0 0 0>
      outlineSettings {
       bUseBrushTransparency=false
       roundingType=HalfHeightRadius
       width=0
       color {colorUseRule=UseColor_Specified, specifiedColor=<rgba 0 0 0 0>}
       cornerRadii=<0 0 0 0>
      }
      tintColor {colorUseRule=UseColor_Specified, specifiedColor=<rgba 1 1 1 1>}
      uVRegion {bIsValid=false, max=<0 0>, min=<0 0>}
     }
    }
   }
   slot {class=BorderSlot, properties {padding=<ltrb 0 0 0 0>}}
  }
  EditorUtilityScrollBox_191 {
   widgetClassPath=&EditorUtilityScrollBox
   bIsVariable=true
   parent=HorizontalBox_388
   widget=EditorUtilityScrollBox_191
   properties {alwaysShowScrollbar=true, scrollbarThickness=<9 1>}
   slot {class=HorizontalBoxSlot, properties {padding=<ltrb 5 0 0 0>, size {sizeRule=Fill, value=1}}}
  }
  ParList_ScaleBox {
   widgetClassPath=&ScaleBox
   parent=EditorUtilityScrollBox_191
   widget=ParList_ScaleBox
   properties {stretch=Fill}
   slot {class=ScrollBoxSlot, properties {padding=<ltrb 10 0 5 0>, size {sizeRule=Fill, value=1}}}
  }
  Border_103 {
   widgetClassPath=&Border
   parent=ParList_ScaleBox
   widget=Border_103
   properties {
    background {
     bIsDynamicallyLoaded=false
     drawAs=Image
     imageType=NoImage
     mirroring=NoMirror
     resourceName=None
     resourceObject=null
     tiling=NoTile
     imageSize=<32 32>
     margin=<ltrb 0 0 0 0>
     outlineSettings {
      bUseBrushTransparency=false
      roundingType=HalfHeightRadius
      width=0
      color {colorUseRule=UseColor_Specified, specifiedColor=<rgba 0 0 0 0>}
      cornerRadii=<0 0 0 0>
     }
     tintColor {colorUseRule=UseColor_Specified, specifiedColor=<rgba 0.0085 0.0085 0.0085 1>}
     uVRegion {bIsValid=false, max=<0 0>, min=<0 0>}
    }
    padding=<ltrb 0 0 0 0>
   }
   slot {class=ScaleBoxSlot, properties {horizontalAlignment=HAlign_Fill, verticalAlignment=VAlign_Fill}}
  }
  CommonListView_Source {
   widgetClassPath=&CommonListView
   bIsVariable=true
   parent=Border_103
   widget=CommonListView_Source
   properties {
    entryWidgetClass=&Anim_Copy_Transform_Selection_C
    scrollBarStyle {
     thickness=8
     draggedThumbImage {
      bIsDynamicallyLoaded=false
      drawAs=RoundedBox
      imageType=NoImage
      mirroring=NoMirror
      resourceName=None
      resourceObject=null
      tiling=NoTile
      imageSize=<0 0>
      margin=<ltrb 0 0 0 0>
      outlineSettings {
       bUseBrushTransparency=false
       roundingType=FixedRadius
       width=0
       color {colorUseRule=UseColor_Specified, specifiedColor=<rgba 0 0 0 0>}
       cornerRadii=<4 4 4 4>
      }
      tintColor {colorUseRule=UseColor_Specified, specifiedColor=<rgba 0.2159 0.2159 0.2159 1>}
      uVRegion {bIsValid=false, max=<0 0>, min=<0 0>}
     }
     horizontalBackgroundImage {
      bIsDynamicallyLoaded=false
      drawAs=NoDrawType
      imageType=NoImage
      mirroring=NoMirror
      resourceName=None
      resourceObject=null
      tiling=NoTile
      imageSize=<0 0>
      margin=<ltrb 0 0 0 0>
      outlineSettings {
       bUseBrushTransparency=false
       roundingType=HalfHeightRadius
       width=0
       color {colorUseRule=UseColor_Specified, specifiedColor=<rgba 0 0 0 0>}
       cornerRadii=<0 0 0 0>
      }
      tintColor {colorUseRule=UseColor_Specified, specifiedColor=<rgba 1 1 1 1>}
      uVRegion {bIsValid=false, max=<0 0>, min=<0 0>}
     }
     horizontalBottomSlotImage {
      bIsDynamicallyLoaded=false
      drawAs=NoDrawType
      imageType=NoImage
      mirroring=NoMirror
      resourceName=None
      resourceObject=null
      tiling=NoTile
      imageSize=<0 0>
      margin=<ltrb 0 0 0 0>
      outlineSettings {
       bUseBrushTransparency=false
       roundingType=HalfHeightRadius
       width=0
       color {colorUseRule=UseColor_Specified, specifiedColor=<rgba 0 0 0 0>}
       cornerRadii=<0 0 0 0>
      }
      tintColor {colorUseRule=UseColor_Specified, specifiedColor=<rgba 1 1 1 1>}
      uVRegion {bIsValid=false, max=<0 0>, min=<0 0>}
     }
     horizontalTopSlotImage {
      bIsDynamicallyLoaded=false
      drawAs=NoDrawType
      imageType=NoImage
      mirroring=NoMirror
      resourceName=None
      resourceObject=null
      tiling=NoTile
      imageSize=<0 0>
      margin=<ltrb 0 0 0 0>
      outlineSettings {
       bUseBrushTransparency=false
       roundingType=HalfHeightRadius
       width=0
       color {colorUseRule=UseColor_Specified, specifiedColor=<rgba 0 0 0 0>}
       cornerRadii=<0 0 0 0>
      }
      tintColor {colorUseRule=UseColor_Specified, specifiedColor=<rgba 1 1 1 1>}
      uVRegion {bIsValid=false, max=<0 0>, min=<0 0>}
     }
     hoveredThumbImage {
      bIsDynamicallyLoaded=false
      drawAs=RoundedBox
      imageType=NoImage
      mirroring=NoMirror
      resourceName=None
      resourceObject=null
      tiling=NoTile
      imageSize=<0 0>
      margin=<ltrb 0 0 0 0>
      outlineSettings {
       bUseBrushTransparency=false
       roundingType=FixedRadius
       width=0
       color {colorUseRule=UseColor_Specified, specifiedColor=<rgba 0 0 0 0>}
       cornerRadii=<4 4 4 4>
      }
      tintColor {colorUseRule=UseColor_Specified, specifiedColor=<rgba 0.2159 0.2159 0.2159 1>}
      uVRegion {bIsValid=false, max=<0 0>, min=<0 0>}
     }
     normalThumbImage {
      bIsDynamicallyLoaded=false
      drawAs=RoundedBox
      imageType=NoImage
      mirroring=NoMirror
      resourceName=None
      resourceObject=null
      tiling=NoTile
      imageSize=<0 0>
      margin=<ltrb 0 0 0 0>
      outlineSettings {
       bUseBrushTransparency=false
       roundingType=FixedRadius
       width=0
       color {colorUseRule=UseColor_Specified, specifiedColor=<rgba 0 0 0 0>}
       cornerRadii=<4 4 4 4>
      }
      tintColor {colorUseRule=UseColor_Specified, specifiedColor=<rgba 0.0953 0.0953 0.0953 1>}
      uVRegion {bIsValid=false, max=<0 0>, min=<0 0>}
     }
     verticalBackgroundImage {
      bIsDynamicallyLoaded=false
      drawAs=NoDrawType
      imageType=NoImage
      mirroring=NoMirror
      resourceName=None
      resourceObject=null
      tiling=NoTile
      imageSize=<0 0>
      margin=<ltrb 0 0 0 0>
      outlineSettings {
       bUseBrushTransparency=false
       roundingType=HalfHeightRadius
       width=0
       color {colorUseRule=UseColor_Specified, specifiedColor=<rgba 0 0 0 0>}
       cornerRadii=<0 0 0 0>
      }
      tintColor {colorUseRule=UseColor_Specified, specifiedColor=<rgba 1 1 1 1>}
      uVRegion {bIsValid=false, max=<0 0>, min=<0 0>}
     }
     verticalBottomSlotImage {
      bIsDynamicallyLoaded=false
      drawAs=NoDrawType
      imageType=NoImage
      mirroring=NoMirror
      resourceName=None
      resourceObject=null
      tiling=NoTile
      imageSize=<0 0>
      margin=<ltrb 0 0 0 0>
      outlineSettings {
       bUseBrushTransparency=false
       roundingType=HalfHeightRadius
       width=0
       color {colorUseRule=UseColor_Specified, specifiedColor=<rgba 0 0 0 0>}
       cornerRadii=<0 0 0 0>
      }
      tintColor {colorUseRule=UseColor_Specified, specifiedColor=<rgba 1 1 1 1>}
      uVRegion {bIsValid=false, max=<0 0>, min=<0 0>}
     }
     verticalTopSlotImage {
      bIsDynamicallyLoaded=false
      drawAs=NoDrawType
      imageType=NoImage
      mirroring=NoMirror
      resourceName=None
      resourceObject=null
      tiling=NoTile
      imageSize=<0 0>
      margin=<ltrb 0 0 0 0>
      outlineSettings {
       bUseBrushTransparency=false
       roundingType=HalfHeightRadius
       width=0
       color {colorUseRule=UseColor_Specified, specifiedColor=<rgba 0 0 0 0>}
       cornerRadii=<0 0 0 0>
      }
      tintColor {colorUseRule=UseColor_Specified, specifiedColor=<rgba 1 1 1 1>}
      uVRegion {bIsValid=false, max=<0 0>, min=<0 0>}
     }
    }
   }
   slot {class=BorderSlot, properties {padding=<ltrb 0 0 0 0>}}
  }
  actionButtons_horiz {
   widgetClassPath=&HorizontalBox
   parent=VerticalBox_42
   widget=actionButtons_horiz
   slot {class=VerticalBoxSlot, properties {verticalAlignment=VAlign_Top, size {sizeRule=Fill, value=1}}}
  }
  ActionButtons_Vert_left {
   widgetClassPath=&VerticalBox
   parent=actionButtons_horiz
   widget=ActionButtons_Vert_left
   slot {class=HorizontalBoxSlot, properties {padding=<ltrb 5 5 5 5>, size {sizeRule=Fill, value=1}}}
  }
  Button_Clear_Dest {
   widgetClassPath=&EditorUtilityButton
   parent=ActionButtons_Vert_left
   widget=Button_Clear_Dest
   properties {
    clipping=ClipToBounds
    widgetStyle {
     clickedSlateSound {resourceObject=null}
     disabled {
      bIsDynamicallyLoaded=false
      drawAs=RoundedBox
      imageType=NoImage
      mirroring=NoMirror
      resourceName=None
      resourceObject=null
      tiling=NoTile
      imageSize=<0 0>
      margin=<ltrb 0 0 0 0>
      outlineSettings {
       bUseBrushTransparency=false
       roundingType=HalfHeightRadius
       width=1
       color {colorUseRule=UseColor_Specified, specifiedColor=<rgba 0.401 0.401 0.401 0.25>}
       cornerRadii=<4 4 4 4>
      }
      tintColor {colorUseRule=UseColor_Specified, specifiedColor=<rgba 0.0194 0.0194 0.0194 1>}
      uVRegion {bIsValid=false, max=<0 0>, min=<0 0>}
     }
     disabledForeground {colorUseRule=UseColor_ColorTable, specifiedColor=<rgba 0 0 0 0>}
     hovered {
      bIsDynamicallyLoaded=false
      drawAs=RoundedBox
      imageType=NoImage
      mirroring=NoMirror
      resourceName=None
      resourceObject=null
      tiling=NoTile
      imageSize=<32 32>
      margin=<ltrb 0 0 0 0>
      outlineSettings {
       bUseBrushTransparency=false
       roundingType=HalfHeightRadius
       width=1
       color {colorUseRule=UseColor_Specified, specifiedColor=<rgba 0 0 0 0>}
       cornerRadii=<4 4 4 4>
      }
      tintColor {colorUseRule=UseColor_Specified, specifiedColor=<rgba 0.3125 0.3125 0.3125 1>}
      uVRegion {bIsValid=false, max=<0 0>, min=<0 0>}
     }
     hoveredForeground {colorUseRule=UseColor_ColorTable, specifiedColor=<rgba 0 0 0 0>}
     hoveredSlateSound {resourceObject=null}
     normal {
      bIsDynamicallyLoaded=false
      drawAs=RoundedBox
      imageType=NoImage
      mirroring=NoMirror
      resourceName=None
      resourceObject=null
      tiling=NoTile
      imageSize=<32 32>
      margin=<ltrb 0 0 0 0>
      outlineSettings {
       bUseBrushTransparency=false
       roundingType=HalfHeightRadius
       width=1
       color {colorUseRule=UseColor_Specified, specifiedColor=<rgba 1 1 1 0.25>}
       cornerRadii=<4 4 4 4>
      }
      tintColor {colorUseRule=UseColor_Specified, specifiedColor=<rgba 0.0194 0.0194 0.0194 1>}
      uVRegion {bIsValid=false, max=<0 0>, min=<0 0>}
     }
     normalForeground {colorUseRule=UseColor_ColorTable, specifiedColor=<rgba 0 0 0 0>}
     normalPadding=<ltrb 12 1.5 12 1.5>
     pressed {
      bIsDynamicallyLoaded=false
      drawAs=RoundedBox
      imageType=NoImage
      mirroring=NoMirror
      resourceName=None
      resourceObject=null
      tiling=NoTile
      imageSize=<32 32>
      margin=<ltrb 0 0 0 0>
      outlineSettings {
       bUseBrushTransparency=false
       roundingType=FixedRadius
       width=1
       color {colorUseRule=UseColor_ColorTable, specifiedColor=<rgba 0 0 0 0>}
       cornerRadii=<4 4 4 4>
      }
      tintColor {colorUseRule=UseColor_Specified, specifiedColor=<rgba 0.0284 0.0284 0.0284 1>}
      uVRegion {bIsValid=false, max=<0 0>, min=<0 0>}
     }
     pressedForeground {colorUseRule=UseColor_ColorTable, specifiedColor=<rgba 0 0 0 0>}
     pressedPadding=<ltrb 12 2.5 12 0.5>
     pressedSlateSound {resourceObject=null}
    }
   }
   slot {class=VerticalBoxSlot, properties {padding=<ltrb 5 5 5 5>}}
  }
  TextBlock_Get_Relatives_1 {
   widgetClassPath=&TextBlock
   parent=Button_Clear_Dest
   widget=TextBlock_Get_Relatives_1
   properties {
    text=clear
    font {
     bForceMonospaced=false
     bMaterialIsStencil=false
     fontMaterial=null
     fontObject=&Roboto
     letterSpacing=0
     monospacedWidth=1
     size=18.75
     skewAmount=0
     typefaceFontName=Light
     outlineSettings {
      bApplyOutlineToDropShadows=false
      bMiteredCorners=false
      bSeparateFillAlpha=false
      outlineMaterial=null
      outlineSize=0
      outlineColor=<rgba 0 0 0 1>
     }
    }
   }
   slot {class=ButtonSlot, properties {verticalAlignment=VAlign_Fill, padding=<ltrb 0 0 0 0>}}
  }
  Button_Copy_Transforms {
   widgetClassPath=&EditorUtilityButton
   parent=ActionButtons_Vert_left
   widget=Button_Copy_Transforms
   properties {
    clipping=ClipToBounds
    widgetStyle {
     clickedSlateSound {resourceObject=null}
     disabled {
      bIsDynamicallyLoaded=false
      drawAs=RoundedBox
      imageType=NoImage
      mirroring=NoMirror
      resourceName=None
      resourceObject=null
      tiling=NoTile
      imageSize=<0 0>
      margin=<ltrb 0 0 0 0>
      outlineSettings {
       bUseBrushTransparency=false
       roundingType=HalfHeightRadius
       width=1
       color {colorUseRule=UseColor_ColorTable, specifiedColor=<rgba 0 0 0 0>}
       cornerRadii=<4 4 4 4>
      }
      tintColor {colorUseRule=UseColor_Specified, specifiedColor=<rgba 0.0395 0.0395 0.0395 1>}
      uVRegion {bIsValid=false, max=<0 0>, min=<0 0>}
     }
     disabledForeground {colorUseRule=UseColor_ColorTable, specifiedColor=<rgba 0 0 0 0>}
     hovered {
      bIsDynamicallyLoaded=false
      drawAs=RoundedBox
      imageType=NoImage
      mirroring=NoMirror
      resourceName=None
      resourceObject=null
      tiling=NoTile
      imageSize=<32 32>
      margin=<ltrb 0 0 0 0>
      outlineSettings {
       bUseBrushTransparency=false
       roundingType=HalfHeightRadius
       width=1
       color {colorUseRule=UseColor_ColorTable, specifiedColor=<rgba 0 0 0 0>}
       cornerRadii=<4 4 4 4>
      }
      tintColor {colorUseRule=UseColor_Specified, specifiedColor=<rgba 0.0953 0.0953 0.0953 1>}
      uVRegion {bIsValid=false, max=<0 0>, min=<0 0>}
     }
     hoveredForeground {colorUseRule=UseColor_ColorTable, specifiedColor=<rgba 0 0 0 0>}
     hoveredSlateSound {resourceObject=null}
     normal {
      bIsDynamicallyLoaded=false
      drawAs=RoundedBox
      imageType=NoImage
      mirroring=NoMirror
      resourceName=None
      resourceObject=null
      tiling=NoTile
      imageSize=<32 32>
      margin=<ltrb 0 0 0 0>
      outlineSettings {
       bUseBrushTransparency=false
       roundingType=HalfHeightRadius
       width=1
       color {colorUseRule=UseColor_ColorTable, specifiedColor=<rgba 0 0 0 0>}
       cornerRadii=<4 4 4 4>
      }
      tintColor {colorUseRule=UseColor_Specified, specifiedColor=<rgba 0.3559 0.2325 0.0604 1>}
      uVRegion {bIsValid=false, max=<0 0>, min=<0 0>}
     }
     normalForeground {colorUseRule=UseColor_ColorTable, specifiedColor=<rgba 0 0 0 0>}
     normalPadding=<ltrb 12 1.5 12 1.5>
     pressed {
      bIsDynamicallyLoaded=false
      drawAs=RoundedBox
      imageType=NoImage
      mirroring=NoMirror
      resourceName=None
      resourceObject=null
      tiling=NoTile
      imageSize=<32 32>
      margin=<ltrb 0 0 0 0>
      outlineSettings {
       bUseBrushTransparency=false
       roundingType=FixedRadius
       width=1
       color {colorUseRule=UseColor_ColorTable, specifiedColor=<rgba 0 0 0 0>}
       cornerRadii=<4 4 4 4>
      }
      tintColor {colorUseRule=UseColor_Specified, specifiedColor=<rgba 0.0284 0.0284 0.0284 1>}
      uVRegion {bIsValid=false, max=<0 0>, min=<0 0>}
     }
     pressedForeground {colorUseRule=UseColor_ColorTable, specifiedColor=<rgba 0 0 0 0>}
     pressedPadding=<ltrb 12 2.5 12 0.5>
     pressedSlateSound {resourceObject=null}
    }
   }
   slot {class=VerticalBoxSlot, properties {padding=<ltrb 5 1 5 5>}}
  }
  TextBlock_Paste_Transforms_1 {
   widgetClassPath=&TextBlock
   parent=Button_Copy_Transforms
   widget=TextBlock_Paste_Transforms_1
   properties {
    text=Copy Transforms
    font {
     bForceMonospaced=false
     bMaterialIsStencil=false
     fontMaterial=null
     fontObject=&Roboto
     letterSpacing=0
     monospacedWidth=1
     size=22.5
     skewAmount=0
     typefaceFontName=Bold
     outlineSettings {
      bApplyOutlineToDropShadows=false
      bMiteredCorners=false
      bSeparateFillAlpha=false
      outlineMaterial=null
      outlineSize=0
      outlineColor=<rgba 0 0 0 1>
     }
    }
   }
   slot {class=ButtonSlot, properties {verticalAlignment=VAlign_Fill, padding=<ltrb 0 0 0 0>}}
  }
  SizeBox_109 {
   widgetClassPath=&SizeBox
   parent=ActionButtons_Vert_left
   widget=SizeBox_109
   properties {heightOverride=30}
   slot {class=VerticalBoxSlot, properties {padding=<ltrb 0 0 0 5>}}
  }
  ScaleBox_80 {widgetClassPath=&ScaleBox, parent=SizeBox_109, widget=ScaleBox_80, slot {class=SizeBoxSlot}}
  VerticalBox_169 {
   widgetClassPath=&VerticalBox
   parent=ScaleBox_80
   widget=VerticalBox_169
   slot {class=ScaleBoxSlot, properties {horizontalAlignment=HAlign_Fill, verticalAlignment=VAlign_Fill}}
  }
  TimeSettingsHorizontalBox {
   widgetClassPath=&HorizontalBox
   parent=VerticalBox_169
   widget=TimeSettingsHorizontalBox
   slot {class=VerticalBoxSlot, properties {verticalAlignment=VAlign_Center}}
  }
  TextInputEnd_1 {
   widgetClassPath=&TextBlock
   parent=TimeSettingsHorizontalBox
   widget=TextInputEnd_1
   properties {
    justification=Right
    text="Paste to: "
    font {
     bForceMonospaced=false
     bMaterialIsStencil=false
     fontMaterial=null
     fontObject=&Roboto
     letterSpacing=0
     monospacedWidth=1
     size=18.75
     skewAmount=0
     typefaceFontName=Regular
     outlineSettings {
      bApplyOutlineToDropShadows=false
      bMiteredCorners=false
      bSeparateFillAlpha=false
      outlineMaterial=null
      outlineSize=0
      outlineColor=<rgba 0 0 0 1>
     }
    }
   }
   slot {class=HorizontalBoxSlot, properties {horizontalAlignment=HAlign_Right, verticalAlignment=VAlign_Top, padding=<ltrb 10 0 0 0>}}
  }
  CurrentFrameCheckBox {
   widgetClassPath=&EditorUtilityCheckBox
   bIsVariable=true
   parent=TimeSettingsHorizontalBox
   widget=CurrentFrameCheckBox
   properties {checkedState=Checked}
   slot {class=HorizontalBoxSlot, properties {verticalAlignment=VAlign_Top, padding=<ltrb 5 0 5 0>}}
  }
  CurrentFrame_TextBlock {
   widgetClassPath=&TextBlock
   parent=CurrentFrameCheckBox
   widget=CurrentFrame_TextBlock
   properties {
    text=Current Frame
    font {
     bForceMonospaced=false
     bMaterialIsStencil=false
     fontMaterial=null
     fontObject=&Roboto
     letterSpacing=0
     monospacedWidth=1
     size=18.75
     skewAmount=0
     typefaceFontName=Regular
     outlineSettings {
      bApplyOutlineToDropShadows=false
      bMiteredCorners=false
      bSeparateFillAlpha=false
      outlineMaterial=null
      outlineSize=0
      outlineColor=<rgba 0 0 0 1>
     }
    }
   }
   slot {class=PanelSlot}
  }
  TimeSliderCheckBox {
   widgetClassPath=&EditorUtilityCheckBox
   bIsVariable=true
   parent=TimeSettingsHorizontalBox
   widget=TimeSliderCheckBox
   properties {horizontalAlignment=HAlign_Center}
   slot {class=HorizontalBoxSlot, properties {horizontalAlignment=HAlign_Center, verticalAlignment=VAlign_Top, padding=<ltrb 8 0 8 0>}}
  }
  LegR_TextBlock_2 {
   widgetClassPath=&TextBlock
   parent=TimeSliderCheckBox
   widget=LegR_TextBlock_2
   properties {
    text=Time Slider
    font {
     bForceMonospaced=false
     bMaterialIsStencil=false
     fontMaterial=null
     fontObject=&Roboto
     letterSpacing=0
     monospacedWidth=1
     size=18.75
     skewAmount=0
     typefaceFontName=Regular
     outlineSettings {
      bApplyOutlineToDropShadows=false
      bMiteredCorners=false
      bSeparateFillAlpha=false
      outlineMaterial=null
      outlineSize=0
      outlineColor=<rgba 0 0 0 1>
     }
    }
   }
   slot {class=PanelSlot}
  }
  SpecifiedRangeCheckBox {
   widgetClassPath=&EditorUtilityCheckBox
   bIsVariable=true
   parent=TimeSettingsHorizontalBox
   widget=SpecifiedRangeCheckBox
   properties {horizontalAlignment=HAlign_Center}
   slot {class=HorizontalBoxSlot, properties {verticalAlignment=VAlign_Top, padding=<ltrb 8 0 10 0>}}
  }
  LegR_TextBlock {
   widgetClassPath=&TextBlock
   parent=SpecifiedRangeCheckBox
   widget=LegR_TextBlock
   properties {
    text=Specified Range
    font {
     bForceMonospaced=false
     bMaterialIsStencil=false
     fontMaterial=null
     fontObject=&Roboto
     letterSpacing=0
     monospacedWidth=1
     size=18.75
     skewAmount=0
     typefaceFontName=Regular
     outlineSettings {
      bApplyOutlineToDropShadows=false
      bMiteredCorners=false
      bSeparateFillAlpha=false
      outlineMaterial=null
      outlineSize=0
      outlineColor=<rgba 0 0 0 1>
     }
    }
   }
   slot {class=PanelSlot}
  }
  SizeBox_51 {
   widgetClassPath=&SizeBox
   parent=VerticalBox_169
   widget=SizeBox_51
   properties {bOverride_MinDesiredHeight=true, maxDesiredHeight=10, minDesiredHeight=10}
   slot {class=VerticalBoxSlot, properties {size {sizeRule=Fill, value=1}}}
  }
  FrameRangeHorizontalBox {
   widgetClassPath=&HorizontalBox
   bIsVariable=true
   parent=SizeBox_51
   widget=FrameRangeHorizontalBox
   properties {visibility=Hidden}
   slot {class=SizeBoxSlot}
  }
  Spacer_40 {
   widgetClassPath=&Spacer
   parent=FrameRangeHorizontalBox
   widget=Spacer_40
   slot {class=HorizontalBoxSlot, properties {size {sizeRule=Fill, value=1}}}
  }
  StartFrame_BTN {
   widgetClassPath=&EditorUtilityButton
   parent=FrameRangeHorizontalBox
   widget=StartFrame_BTN
   properties {widgetStyle {
     clickedSlateSound {resourceObject=null}
     disabled {
      bIsDynamicallyLoaded=false
      drawAs=RoundedBox
      imageType=NoImage
      mirroring=NoMirror
      resourceName=None
      resourceObject=null
      tiling=NoTile
      imageSize=<0 0>
      margin=<ltrb 0 0 0 0>
      outlineSettings {
       bUseBrushTransparency=false
       roundingType=FixedRadius
       width=1
       color {colorUseRule=UseColor_ColorTable, specifiedColor=<rgba 0 0 0 0>}
       cornerRadii=<4 4 4 4>
      }
      tintColor {colorUseRule=UseColor_Specified, specifiedColor=<rgba 0.0395 0.0395 0.0395 1>}
      uVRegion {bIsValid=false, max=<0 0>, min=<0 0>}
     }
     disabledForeground {colorUseRule=UseColor_ColorTable, specifiedColor=<rgba 0 0 0 0>}
     hovered {
      bIsDynamicallyLoaded=false
      drawAs=RoundedBox
      imageType=NoImage
      mirroring=NoMirror
      resourceName=None
      resourceObject=null
      tiling=NoTile
      imageSize=<32 8>
      margin=<ltrb 0 0 0 0>
      outlineSettings {
       bUseBrushTransparency=false
       roundingType=FixedRadius
       width=1
       color {colorUseRule=UseColor_ColorTable, specifiedColor=<rgba 0 0 0 0>}
       cornerRadii=<4 4 4 4>
      }
      tintColor {colorUseRule=UseColor_Specified, specifiedColor=<rgba 0.0953 0.0953 0.0953 1>}
      uVRegion {bIsValid=false, max=<0 0>, min=<0 0>}
     }
     hoveredForeground {colorUseRule=UseColor_ColorTable, specifiedColor=<rgba 0 0 0 0>}
     hoveredSlateSound {resourceObject=null}
     normal {
      bIsDynamicallyLoaded=false
      drawAs=RoundedBox
      imageType=NoImage
      mirroring=NoMirror
      resourceName=None
      resourceObject=null
      tiling=NoTile
      imageSize=<32 8>
      margin=<ltrb 0 0 0 0>
      outlineSettings {
       bUseBrushTransparency=false
       roundingType=FixedRadius
       width=1
       color {colorUseRule=UseColor_ColorTable, specifiedColor=<rgba 0 0 0 0>}
       cornerRadii=<4 4 4 4>
      }
      tintColor {colorUseRule=UseColor_Specified, specifiedColor=<rgba 0.0395 0.0395 0.0395 1>}
      uVRegion {bIsValid=false, max=<0 0>, min=<0 0>}
     }
     normalForeground {colorUseRule=UseColor_ColorTable, specifiedColor=<rgba 0 0 0 0>}
     normalPadding=<ltrb 12 1.5 12 1.5>
     pressed {
      bIsDynamicallyLoaded=false
      drawAs=RoundedBox
      imageType=NoImage
      mirroring=NoMirror
      resourceName=None
      resourceObject=null
      tiling=NoTile
      imageSize=<32 8>
      margin=<ltrb 0 0 0 0>
      outlineSettings {
       bUseBrushTransparency=false
       roundingType=FixedRadius
       width=1
       color {colorUseRule=UseColor_ColorTable, specifiedColor=<rgba 0 0 0 0>}
       cornerRadii=<4 4 4 4>
      }
      tintColor {colorUseRule=UseColor_Specified, specifiedColor=<rgba 0.0284 0.0284 0.0284 1>}
      uVRegion {bIsValid=false, max=<0 0>, min=<0 0>}
     }
     pressedForeground {colorUseRule=UseColor_ColorTable, specifiedColor=<rgba 0 0 0 0>}
     pressedPadding=<ltrb 12 2.5 12 0.5>
     pressedSlateSound {resourceObject=null}
    }}
   slot {class=HorizontalBoxSlot, properties {padding=<ltrb 5 0 0 0>, size {sizeRule=Fill, value=1}}}
  }
  TextBlock_136 {
   widgetClassPath=&TextBlock
   parent=StartFrame_BTN
   widget=TextBlock_136
   properties {
    justification=Right
    text=Start:
    font {
     bForceMonospaced=false
     bMaterialIsStencil=false
     fontMaterial=null
     fontObject=&Roboto
     letterSpacing=0
     monospacedWidth=1
     size=18.75
     skewAmount=0
     typefaceFontName=Regular
     outlineSettings {
      bApplyOutlineToDropShadows=false
      bMiteredCorners=false
      bSeparateFillAlpha=false
      outlineMaterial=null
      outlineSize=0
      outlineColor=<rgba 0 0 0 1>
     }
    }
   }
   slot {class=ButtonSlot, properties {padding=<ltrb 0 2 0 2>}}
  }
  StartFrameTextBox {
   widgetClassPath=&EditorUtilityEditableTextBox
   bIsVariable=true
   parent=FrameRangeHorizontalBox
   widget=StartFrameTextBox
   properties {
    hintText=0
    widgetStyle {
     backgroundColor {colorUseRule=UseColor_Specified, specifiedColor=<rgba 1 1 1 1>}
     backgroundImageFocused {
      bIsDynamicallyLoaded=false
      drawAs=RoundedBox
      imageType=NoImage
      mirroring=NoMirror
      resourceName=None
      resourceObject=null
      tiling=NoTile
      imageSize=<0 0>
      margin=<ltrb 0 0 0 0>
      outlineSettings {
       bUseBrushTransparency=false
       roundingType=FixedRadius
       width=1
       color {colorUseRule=UseColor_ColorTable, specifiedColor=<rgba 0 0 0 0>}
       cornerRadii=<4 4 4 4>
      }
      tintColor {colorUseRule=UseColor_Specified, specifiedColor=<rgba 0.0048 0.0048 0.0048 1>}
      uVRegion {bIsValid=false, max=<0 0>, min=<0 0>}
     }
     backgroundImageHovered {
      bIsDynamicallyLoaded=false
      drawAs=RoundedBox
      imageType=NoImage
      mirroring=NoMirror
      resourceName=None
      resourceObject=null
      tiling=NoTile
      imageSize=<0 0>
      margin=<ltrb 0 0 0 0>
      outlineSettings {
       bUseBrushTransparency=false
       roundingType=FixedRadius
       width=1
       color {colorUseRule=UseColor_ColorTable, specifiedColor=<rgba 0 0 0 0>}
       cornerRadii=<4 4 4 4>
      }
      tintColor {colorUseRule=UseColor_Specified, specifiedColor=<rgba 0.0048 0.0048 0.0048 1>}
      uVRegion {bIsValid=false, max=<0 0>, min=<0 0>}
     }
     backgroundImageNormal {
      bIsDynamicallyLoaded=false
      drawAs=RoundedBox
      imageType=NoImage
      mirroring=NoMirror
      resourceName=None
      resourceObject=null
      tiling=NoTile
      imageSize=<0 0>
      margin=<ltrb 0 0 0 0>
      outlineSettings {
       bUseBrushTransparency=false
       roundingType=FixedRadius
       width=1
       color {colorUseRule=UseColor_ColorTable, specifiedColor=<rgba 0 0 0 0>}
       cornerRadii=<4 4 4 4>
      }
      tintColor {colorUseRule=UseColor_Specified, specifiedColor=<rgba 0.0048 0.0048 0.0048 1>}
      uVRegion {bIsValid=false, max=<0 0>, min=<0 0>}
     }
     backgroundImageReadOnly {
      bIsDynamicallyLoaded=false
      drawAs=RoundedBox
      imageType=NoImage
      mirroring=NoMirror
      resourceName=None
      resourceObject=null
      tiling=NoTile
      imageSize=<0 0>
      margin=<ltrb 0 0 0 0>
      outlineSettings {
       bUseBrushTransparency=false
       roundingType=FixedRadius
       width=1
       color {colorUseRule=UseColor_ColorTable, specifiedColor=<rgba 0 0 0 0>}
       cornerRadii=<4 4 4 4>
      }
      tintColor {colorUseRule=UseColor_Specified, specifiedColor=<rgba 0.0284 0.0284 0.0284 1>}
      uVRegion {bIsValid=false, max=<0 0>, min=<0 0>}
     }
     focusedForegroundColor {colorUseRule=UseColor_Specified, specifiedColor=<rgba 1 1 1 1>}
     foregroundColor {colorUseRule=UseColor_Specified, specifiedColor=<rgba 0.5271 0.5271 0.5271 1>}
     hScrollBarPadding=<ltrb 0 0 0 0>
     padding=<ltrb 8 3 8 4>
     readOnlyForegroundColor {colorUseRule=UseColor_Foreground, specifiedColor=<rgba 1 0 1 1>}
     scrollBarStyle {
      thickness=8
      draggedThumbImage {
       bIsDynamicallyLoaded=false
       drawAs=RoundedBox
       imageType=NoImage
       mirroring=NoMirror
       resourceName=None
       resourceObject=null
       tiling=NoTile
       imageSize=<0 0>
       margin=<ltrb 0 0 0 0>
       outlineSettings {
        bUseBrushTransparency=false
        roundingType=FixedRadius
        width=0
        color {colorUseRule=UseColor_Specified, specifiedColor=<rgba 0 0 0 0>}
        cornerRadii=<4 4 4 4>
       }
       tintColor {colorUseRule=UseColor_Specified, specifiedColor=<rgba 0.2159 0.2159 0.2159 1>}
       uVRegion {bIsValid=false, max=<0 0>, min=<0 0>}
      }
      horizontalBackgroundImage {
       bIsDynamicallyLoaded=false
       drawAs=NoDrawType
       imageType=NoImage
       mirroring=NoMirror
       resourceName=None
       resourceObject=null
       tiling=NoTile
       imageSize=<0 0>
       margin=<ltrb 0 0 0 0>
       outlineSettings {
        bUseBrushTransparency=false
        roundingType=HalfHeightRadius
        width=0
        color {colorUseRule=UseColor_Specified, specifiedColor=<rgba 0 0 0 0>}
        cornerRadii=<0 0 0 0>
       }
       tintColor {colorUseRule=UseColor_Specified, specifiedColor=<rgba 1 1 1 1>}
       uVRegion {bIsValid=false, max=<0 0>, min=<0 0>}
      }
      horizontalBottomSlotImage {
       bIsDynamicallyLoaded=false
       drawAs=NoDrawType
       imageType=NoImage
       mirroring=NoMirror
       resourceName=None
       resourceObject=null
       tiling=NoTile
       imageSize=<0 0>
       margin=<ltrb 0 0 0 0>
       outlineSettings {
        bUseBrushTransparency=false
        roundingType=HalfHeightRadius
        width=0
        color {colorUseRule=UseColor_Specified, specifiedColor=<rgba 0 0 0 0>}
        cornerRadii=<0 0 0 0>
       }
       tintColor {colorUseRule=UseColor_Specified, specifiedColor=<rgba 1 1 1 1>}
       uVRegion {bIsValid=false, max=<0 0>, min=<0 0>}
      }
      horizontalTopSlotImage {
       bIsDynamicallyLoaded=false
       drawAs=NoDrawType
       imageType=NoImage
       mirroring=NoMirror
       resourceName=None
       resourceObject=null
       tiling=NoTile
       imageSize=<0 0>
       margin=<ltrb 0 0 0 0>
       outlineSettings {
        bUseBrushTransparency=false
        roundingType=HalfHeightRadius
        width=0
        color {colorUseRule=UseColor_Specified, specifiedColor=<rgba 0 0 0 0>}
        cornerRadii=<0 0 0 0>
       }
       tintColor {colorUseRule=UseColor_Specified, specifiedColor=<rgba 1 1 1 1>}
       uVRegion {bIsValid=false, max=<0 0>, min=<0 0>}
      }
      hoveredThumbImage {
       bIsDynamicallyLoaded=false
       drawAs=RoundedBox
       imageType=NoImage
       mirroring=NoMirror
       resourceName=None
       resourceObject=null
       tiling=NoTile
       imageSize=<0 0>
       margin=<ltrb 0 0 0 0>
       outlineSettings {
        bUseBrushTransparency=false
        roundingType=FixedRadius
        width=0
        color {colorUseRule=UseColor_Specified, specifiedColor=<rgba 0 0 0 0>}
        cornerRadii=<4 4 4 4>
       }
       tintColor {colorUseRule=UseColor_Specified, specifiedColor=<rgba 0.2159 0.2159 0.2159 1>}
       uVRegion {bIsValid=false, max=<0 0>, min=<0 0>}
      }
      normalThumbImage {
       bIsDynamicallyLoaded=false
       drawAs=RoundedBox
       imageType=NoImage
       mirroring=NoMirror
       resourceName=None
       resourceObject=null
       tiling=NoTile
       imageSize=<0 0>
       margin=<ltrb 0 0 0 0>
       outlineSettings {
        bUseBrushTransparency=false
        roundingType=FixedRadius
        width=0
        color {colorUseRule=UseColor_Specified, specifiedColor=<rgba 0 0 0 0>}
        cornerRadii=<4 4 4 4>
       }
       tintColor {colorUseRule=UseColor_Specified, specifiedColor=<rgba 0.0953 0.0953 0.0953 1>}
       uVRegion {bIsValid=false, max=<0 0>, min=<0 0>}
      }
      verticalBackgroundImage {
       bIsDynamicallyLoaded=false
       drawAs=NoDrawType
       imageType=NoImage
       mirroring=NoMirror
       resourceName=None
       resourceObject=null
       tiling=NoTile
       imageSize=<0 0>
       margin=<ltrb 0 0 0 0>
       outlineSettings {
        bUseBrushTransparency=false
        roundingType=HalfHeightRadius
        width=0
        color {colorUseRule=UseColor_Specified, specifiedColor=<rgba 0 0 0 0>}
        cornerRadii=<0 0 0 0>
       }
       tintColor {colorUseRule=UseColor_Specified, specifiedColor=<rgba 1 1 1 1>}
       uVRegion {bIsValid=false, max=<0 0>, min=<0 0>}
      }
      verticalBottomSlotImage {
       bIsDynamicallyLoaded=false
       drawAs=NoDrawType
       imageType=NoImage
       mirroring=NoMirror
       resourceName=None
       resourceObject=null
       tiling=NoTile
       imageSize=<0 0>
       margin=<ltrb 0 0 0 0>
       outlineSettings {
        bUseBrushTransparency=false
        roundingType=HalfHeightRadius
        width=0
        color {colorUseRule=UseColor_Specified, specifiedColor=<rgba 0 0 0 0>}
        cornerRadii=<0 0 0 0>
       }
       tintColor {colorUseRule=UseColor_Specified, specifiedColor=<rgba 1 1 1 1>}
       uVRegion {bIsValid=false, max=<0 0>, min=<0 0>}
      }
      verticalTopSlotImage {
       bIsDynamicallyLoaded=false
       drawAs=NoDrawType
       imageType=NoImage
       mirroring=NoMirror
       resourceName=None
       resourceObject=null
       tiling=NoTile
       imageSize=<0 0>
       margin=<ltrb 0 0 0 0>
       outlineSettings {
        bUseBrushTransparency=false
        roundingType=HalfHeightRadius
        width=0
        color {colorUseRule=UseColor_Specified, specifiedColor=<rgba 0 0 0 0>}
        cornerRadii=<0 0 0 0>
       }
       tintColor {colorUseRule=UseColor_Specified, specifiedColor=<rgba 1 1 1 1>}
       uVRegion {bIsValid=false, max=<0 0>, min=<0 0>}
      }
     }
     textStyle {
      overflowPolicy=Clip
      transformPolicy=None
      colorAndOpacity {colorUseRule=UseColor_Foreground, specifiedColor=<rgba 1 0 1 1>}
      font {
       bForceMonospaced=false
       bMaterialIsStencil=false
       fontMaterial=null
       fontObject=null
       letterSpacing=0
       monospacedWidth=1
       size=15
       skewAmount=0
       typefaceFontName=Regular
       outlineSettings {
        bApplyOutlineToDropShadows=false
        bMiteredCorners=false
        bSeparateFillAlpha=false
        outlineMaterial=null
        outlineSize=0
        outlineColor=<rgba 0 0 0 1>
       }
      }
      highlightColor {colorUseRule=UseColor_ColorTable, specifiedColor=<rgba 0 0 0 0>}
      highlightShape {
       bIsDynamicallyLoaded=false
       drawAs=Image
       imageType=NoImage
       mirroring=NoMirror
       resourceName=None
       resourceObject=null
       tiling=NoTile
       imageSize=<0 0>
       margin=<ltrb 0 0 0 0>
       outlineSettings {
        bUseBrushTransparency=false
        roundingType=HalfHeightRadius
        width=0
        color {colorUseRule=UseColor_Specified, specifiedColor=<rgba 0 0 0 0>}
        cornerRadii=<0 0 0 0>
       }
       tintColor {colorUseRule=UseColor_ColorTable, specifiedColor=<rgba 0 0 0 0>}
       uVRegion {bIsValid=false, max=<0 0>, min=<0 0>}
      }
      selectedBackgroundColor {colorUseRule=UseColor_ColorTable, specifiedColor=<rgba 0 0 0 0>}
      shadowColorAndOpacity=<rgba 0 0 0 1>
      shadowOffset=<0 0>
      strikeBrush {
       bIsDynamicallyLoaded=false
       drawAs=Image
       imageType=NoImage
       mirroring=NoMirror
       resourceName=None
       resourceObject=null
       tiling=NoTile
       imageSize=<32 32>
       margin=<ltrb 0 0 0 0>
       outlineSettings {
        bUseBrushTransparency=false
        roundingType=HalfHeightRadius
        width=0
        color {colorUseRule=UseColor_Specified, specifiedColor=<rgba 0 0 0 0>}
        cornerRadii=<0 0 0 0>
       }
       tintColor {colorUseRule=UseColor_Specified, specifiedColor=<rgba 1 1 1 1>}
       uVRegion {bIsValid=false, max=<0 0>, min=<0 0>}
      }
      underlineBrush {
       bIsDynamicallyLoaded=false
       drawAs=Image
       imageType=NoImage
       mirroring=NoMirror
       resourceName=None
       resourceObject=null
       tiling=NoTile
       imageSize=<32 32>
       margin=<ltrb 0 0 0 0>
       outlineSettings {
        bUseBrushTransparency=false
        roundingType=HalfHeightRadius
        width=0
        color {colorUseRule=UseColor_Specified, specifiedColor=<rgba 0 0 0 0>}
        cornerRadii=<0 0 0 0>
       }
       tintColor {colorUseRule=UseColor_Specified, specifiedColor=<rgba 1 1 1 1>}
       uVRegion {bIsValid=false, max=<0 0>, min=<0 0>}
      }
     }
     vScrollBarPadding=<ltrb 0 0 0 0>
    }
   }
   slot {class=HorizontalBoxSlot, properties {size {sizeRule=Fill, value=1}}}
  }
  EndFrame_BTN {
   widgetClassPath=&EditorUtilityButton
   parent=FrameRangeHorizontalBox
   widget=EndFrame_BTN
   properties {widgetStyle {
     clickedSlateSound {resourceObject=null}
     disabled {
      bIsDynamicallyLoaded=false
      drawAs=RoundedBox
      imageType=NoImage
      mirroring=NoMirror
      resourceName=None
      resourceObject=null
      tiling=NoTile
      imageSize=<0 0>
      margin=<ltrb 0 0 0 0>
      outlineSettings {
       bUseBrushTransparency=false
       roundingType=FixedRadius
       width=1
       color {colorUseRule=UseColor_ColorTable, specifiedColor=<rgba 0 0 0 0>}
       cornerRadii=<4 4 4 4>
      }
      tintColor {colorUseRule=UseColor_Specified, specifiedColor=<rgba 0.0395 0.0395 0.0395 1>}
      uVRegion {bIsValid=false, max=<0 0>, min=<0 0>}
     }
     disabledForeground {colorUseRule=UseColor_ColorTable, specifiedColor=<rgba 0 0 0 0>}
     hovered {
      bIsDynamicallyLoaded=false
      drawAs=RoundedBox
      imageType=NoImage
      mirroring=NoMirror
      resourceName=None
      resourceObject=null
      tiling=NoTile
      imageSize=<32 8>
      margin=<ltrb 0 0 0 0>
      outlineSettings {
       bUseBrushTransparency=false
       roundingType=FixedRadius
       width=1
       color {colorUseRule=UseColor_ColorTable, specifiedColor=<rgba 0 0 0 0>}
       cornerRadii=<4 4 4 4>
      }
      tintColor {colorUseRule=UseColor_Specified, specifiedColor=<rgba 0.0953 0.0953 0.0953 1>}
      uVRegion {bIsValid=false, max=<0 0>, min=<0 0>}
     }
     hoveredForeground {colorUseRule=UseColor_ColorTable, specifiedColor=<rgba 0 0 0 0>}
     hoveredSlateSound {resourceObject=null}
     normal {
      bIsDynamicallyLoaded=false
      drawAs=RoundedBox
      imageType=NoImage
      mirroring=NoMirror
      resourceName=None
      resourceObject=null
      tiling=NoTile
      imageSize=<32 8>
      margin=<ltrb 0 0 0 0>
      outlineSettings {
       bUseBrushTransparency=false
       roundingType=FixedRadius
       width=1
       color {colorUseRule=UseColor_ColorTable, specifiedColor=<rgba 0 0 0 0>}
       cornerRadii=<4 4 4 4>
      }
      tintColor {colorUseRule=UseColor_Specified, specifiedColor=<rgba 0.0395 0.0395 0.0395 1>}
      uVRegion {bIsValid=false, max=<0 0>, min=<0 0>}
     }
     normalForeground {colorUseRule=UseColor_ColorTable, specifiedColor=<rgba 0 0 0 0>}
     normalPadding=<ltrb 12 1.5 12 1.5>
     pressed {
      bIsDynamicallyLoaded=false
      drawAs=RoundedBox
      imageType=NoImage
      mirroring=NoMirror
      resourceName=None
      resourceObject=null
      tiling=NoTile
      imageSize=<32 8>
      margin=<ltrb 0 0 0 0>
      outlineSettings {
       bUseBrushTransparency=false
       roundingType=FixedRadius
       width=1
       color {colorUseRule=UseColor_ColorTable, specifiedColor=<rgba 0 0 0 0>}
       cornerRadii=<4 4 4 4>
      }
      tintColor {colorUseRule=UseColor_Specified, specifiedColor=<rgba 0.0284 0.0284 0.0284 1>}
      uVRegion {bIsValid=false, max=<0 0>, min=<0 0>}
     }
     pressedForeground {colorUseRule=UseColor_ColorTable, specifiedColor=<rgba 0 0 0 0>}
     pressedPadding=<ltrb 12 2.5 12 0.5>
     pressedSlateSound {resourceObject=null}
    }}
   slot {class=HorizontalBoxSlot, properties {padding=<ltrb 5 0 0 0>, size {sizeRule=Fill, value=1}}}
  }
  TextBlock_2 {
   widgetClassPath=&TextBlock
   parent=EndFrame_BTN
   widget=TextBlock_2
   properties {
    text=End:
    font {
     bForceMonospaced=false
     bMaterialIsStencil=false
     fontMaterial=null
     fontObject=&Roboto
     letterSpacing=0
     monospacedWidth=1
     size=18.75
     skewAmount=0
     typefaceFontName=Regular
     outlineSettings {
      bApplyOutlineToDropShadows=false
      bMiteredCorners=false
      bSeparateFillAlpha=false
      outlineMaterial=null
      outlineSize=0
      outlineColor=<rgba 0 0 0 1>
     }
    }
   }
   slot {class=ButtonSlot, properties {padding=<ltrb 0 2 0 2>}}
  }
  EndFrameTextBox {
   widgetClassPath=&EditorUtilityEditableTextBox
   bIsVariable=true
   parent=FrameRangeHorizontalBox
   widget=EndFrameTextBox
   properties {
    hintText=100
    widgetStyle {
     backgroundColor {colorUseRule=UseColor_Specified, specifiedColor=<rgba 1 1 1 1>}
     backgroundImageFocused {
      bIsDynamicallyLoaded=false
      drawAs=RoundedBox
      imageType=NoImage
      mirroring=NoMirror
      resourceName=None
      resourceObject=null
      tiling=NoTile
      imageSize=<0 0>
      margin=<ltrb 0 0 0 0>
      outlineSettings {
       bUseBrushTransparency=false
       roundingType=FixedRadius
       width=1
       color {colorUseRule=UseColor_ColorTable, specifiedColor=<rgba 0 0 0 0>}
       cornerRadii=<4 4 4 4>
      }
      tintColor {colorUseRule=UseColor_Specified, specifiedColor=<rgba 0.0048 0.0048 0.0048 1>}
      uVRegion {bIsValid=false, max=<0 0>, min=<0 0>}
     }
     backgroundImageHovered {
      bIsDynamicallyLoaded=false
      drawAs=RoundedBox
      imageType=NoImage
      mirroring=NoMirror
      resourceName=None
      resourceObject=null
      tiling=NoTile
      imageSize=<0 0>
      margin=<ltrb 0 0 0 0>
      outlineSettings {
       bUseBrushTransparency=false
       roundingType=FixedRadius
       width=1
       color {colorUseRule=UseColor_ColorTable, specifiedColor=<rgba 0 0 0 0>}
       cornerRadii=<4 4 4 4>
      }
      tintColor {colorUseRule=UseColor_Specified, specifiedColor=<rgba 0.0048 0.0048 0.0048 1>}
      uVRegion {bIsValid=false, max=<0 0>, min=<0 0>}
     }
     backgroundImageNormal {
      bIsDynamicallyLoaded=false
      drawAs=RoundedBox
      imageType=NoImage
      mirroring=NoMirror
      resourceName=None
      resourceObject=null
      tiling=NoTile
      imageSize=<0 0>
      margin=<ltrb 0 0 0 0>
      outlineSettings {
       bUseBrushTransparency=false
       roundingType=FixedRadius
       width=1
       color {colorUseRule=UseColor_ColorTable, specifiedColor=<rgba 0 0 0 0>}
       cornerRadii=<4 4 4 4>
      }
      tintColor {colorUseRule=UseColor_Specified, specifiedColor=<rgba 0.0048 0.0048 0.0048 1>}
      uVRegion {bIsValid=false, max=<0 0>, min=<0 0>}
     }
     backgroundImageReadOnly {
      bIsDynamicallyLoaded=false
      drawAs=RoundedBox
      imageType=NoImage
      mirroring=NoMirror
      resourceName=None
      resourceObject=null
      tiling=NoTile
      imageSize=<0 0>
      margin=<ltrb 0 0 0 0>
      outlineSettings {
       bUseBrushTransparency=false
       roundingType=FixedRadius
       width=1
       color {colorUseRule=UseColor_ColorTable, specifiedColor=<rgba 0 0 0 0>}
       cornerRadii=<4 4 4 4>
      }
      tintColor {colorUseRule=UseColor_Specified, specifiedColor=<rgba 0.0284 0.0284 0.0284 1>}
      uVRegion {bIsValid=false, max=<0 0>, min=<0 0>}
     }
     focusedForegroundColor {colorUseRule=UseColor_Specified, specifiedColor=<rgba 1 1 1 1>}
     foregroundColor {colorUseRule=UseColor_Specified, specifiedColor=<rgba 0.5271 0.5271 0.5271 1>}
     hScrollBarPadding=<ltrb 0 0 0 0>
     padding=<ltrb 8 3 8 4>
     readOnlyForegroundColor {colorUseRule=UseColor_Foreground, specifiedColor=<rgba 1 0 1 1>}
     scrollBarStyle {
      thickness=8
      draggedThumbImage {
       bIsDynamicallyLoaded=false
       drawAs=RoundedBox
       imageType=NoImage
       mirroring=NoMirror
       resourceName=None
       resourceObject=null
       tiling=NoTile
       imageSize=<0 0>
       margin=<ltrb 0 0 0 0>
       outlineSettings {
        bUseBrushTransparency=false
        roundingType=FixedRadius
        width=0
        color {colorUseRule=UseColor_Specified, specifiedColor=<rgba 0 0 0 0>}
        cornerRadii=<4 4 4 4>
       }
       tintColor {colorUseRule=UseColor_Specified, specifiedColor=<rgba 0.2159 0.2159 0.2159 1>}
       uVRegion {bIsValid=false, max=<0 0>, min=<0 0>}
      }
      horizontalBackgroundImage {
       bIsDynamicallyLoaded=false
       drawAs=NoDrawType
       imageType=NoImage
       mirroring=NoMirror
       resourceName=None
       resourceObject=null
       tiling=NoTile
       imageSize=<0 0>
       margin=<ltrb 0 0 0 0>
       outlineSettings {
        bUseBrushTransparency=false
        roundingType=HalfHeightRadius
        width=0
        color {colorUseRule=UseColor_Specified, specifiedColor=<rgba 0 0 0 0>}
        cornerRadii=<0 0 0 0>
       }
       tintColor {colorUseRule=UseColor_Specified, specifiedColor=<rgba 1 1 1 1>}
       uVRegion {bIsValid=false, max=<0 0>, min=<0 0>}
      }
      horizontalBottomSlotImage {
       bIsDynamicallyLoaded=false
       drawAs=NoDrawType
       imageType=NoImage
       mirroring=NoMirror
       resourceName=None
       resourceObject=null
       tiling=NoTile
       imageSize=<0 0>
       margin=<ltrb 0 0 0 0>
       outlineSettings {
        bUseBrushTransparency=false
        roundingType=HalfHeightRadius
        width=0
        color {colorUseRule=UseColor_Specified, specifiedColor=<rgba 0 0 0 0>}
        cornerRadii=<0 0 0 0>
       }
       tintColor {colorUseRule=UseColor_Specified, specifiedColor=<rgba 1 1 1 1>}
       uVRegion {bIsValid=false, max=<0 0>, min=<0 0>}
      }
      horizontalTopSlotImage {
       bIsDynamicallyLoaded=false
       drawAs=NoDrawType
       imageType=NoImage
       mirroring=NoMirror
       resourceName=None
       resourceObject=null
       tiling=NoTile
       imageSize=<0 0>
       margin=<ltrb 0 0 0 0>
       outlineSettings {
        bUseBrushTransparency=false
        roundingType=HalfHeightRadius
        width=0
        color {colorUseRule=UseColor_Specified, specifiedColor=<rgba 0 0 0 0>}
        cornerRadii=<0 0 0 0>
       }
       tintColor {colorUseRule=UseColor_Specified, specifiedColor=<rgba 1 1 1 1>}
       uVRegion {bIsValid=false, max=<0 0>, min=<0 0>}
      }
      hoveredThumbImage {
       bIsDynamicallyLoaded=false
       drawAs=RoundedBox
       imageType=NoImage
       mirroring=NoMirror
       resourceName=None
       resourceObject=null
       tiling=NoTile
       imageSize=<0 0>
       margin=<ltrb 0 0 0 0>
       outlineSettings {
        bUseBrushTransparency=false
        roundingType=FixedRadius
        width=0
        color {colorUseRule=UseColor_Specified, specifiedColor=<rgba 0 0 0 0>}
        cornerRadii=<4 4 4 4>
       }
       tintColor {colorUseRule=UseColor_Specified, specifiedColor=<rgba 0.2159 0.2159 0.2159 1>}
       uVRegion {bIsValid=false, max=<0 0>, min=<0 0>}
      }
      normalThumbImage {
       bIsDynamicallyLoaded=false
       drawAs=RoundedBox
       imageType=NoImage
       mirroring=NoMirror
       resourceName=None
       resourceObject=null
       tiling=NoTile
       imageSize=<0 0>
       margin=<ltrb 0 0 0 0>
       outlineSettings {
        bUseBrushTransparency=false
        roundingType=FixedRadius
        width=0
        color {colorUseRule=UseColor_Specified, specifiedColor=<rgba 0 0 0 0>}
        cornerRadii=<4 4 4 4>
       }
       tintColor {colorUseRule=UseColor_Specified, specifiedColor=<rgba 0.0953 0.0953 0.0953 1>}
       uVRegion {bIsValid=false, max=<0 0>, min=<0 0>}
      }
      verticalBackgroundImage {
       bIsDynamicallyLoaded=false
       drawAs=NoDrawType
       imageType=NoImage
       mirroring=NoMirror
       resourceName=None
       resourceObject=null
       tiling=NoTile
       imageSize=<0 0>
       margin=<ltrb 0 0 0 0>
       outlineSettings {
        bUseBrushTransparency=false
        roundingType=HalfHeightRadius
        width=0
        color {colorUseRule=UseColor_Specified, specifiedColor=<rgba 0 0 0 0>}
        cornerRadii=<0 0 0 0>
       }
       tintColor {colorUseRule=UseColor_Specified, specifiedColor=<rgba 1 1 1 1>}
       uVRegion {bIsValid=false, max=<0 0>, min=<0 0>}
      }
      verticalBottomSlotImage {
       bIsDynamicallyLoaded=false
       drawAs=NoDrawType
       imageType=NoImage
       mirroring=NoMirror
       resourceName=None
       resourceObject=null
       tiling=NoTile
       imageSize=<0 0>
       margin=<ltrb 0 0 0 0>
       outlineSettings {
        bUseBrushTransparency=false
        roundingType=HalfHeightRadius
        width=0
        color {colorUseRule=UseColor_Specified, specifiedColor=<rgba 0 0 0 0>}
        cornerRadii=<0 0 0 0>
       }
       tintColor {colorUseRule=UseColor_Specified, specifiedColor=<rgba 1 1 1 1>}
       uVRegion {bIsValid=false, max=<0 0>, min=<0 0>}
      }
      verticalTopSlotImage {
       bIsDynamicallyLoaded=false
       drawAs=NoDrawType
       imageType=NoImage
       mirroring=NoMirror
       resourceName=None
       resourceObject=null
       tiling=NoTile
       imageSize=<0 0>
       margin=<ltrb 0 0 0 0>
       outlineSettings {
        bUseBrushTransparency=false
        roundingType=HalfHeightRadius
        width=0
        color {colorUseRule=UseColor_Specified, specifiedColor=<rgba 0 0 0 0>}
        cornerRadii=<0 0 0 0>
       }
       tintColor {colorUseRule=UseColor_Specified, specifiedColor=<rgba 1 1 1 1>}
       uVRegion {bIsValid=false, max=<0 0>, min=<0 0>}
      }
     }
     textStyle {
      overflowPolicy=Clip
      transformPolicy=None
      colorAndOpacity {colorUseRule=UseColor_Foreground, specifiedColor=<rgba 1 0 1 1>}
      font {
       bForceMonospaced=false
       bMaterialIsStencil=false
       fontMaterial=null
       fontObject=null
       letterSpacing=0
       monospacedWidth=1
       size=15
       skewAmount=0
       typefaceFontName=Regular
       outlineSettings {
        bApplyOutlineToDropShadows=false
        bMiteredCorners=false
        bSeparateFillAlpha=false
        outlineMaterial=null
        outlineSize=0
        outlineColor=<rgba 0 0 0 1>
       }
      }
      highlightColor {colorUseRule=UseColor_ColorTable, specifiedColor=<rgba 0 0 0 0>}
      highlightShape {
       bIsDynamicallyLoaded=false
       drawAs=Image
       imageType=NoImage
       mirroring=NoMirror
       resourceName=None
       resourceObject=null
       tiling=NoTile
       imageSize=<0 0>
       margin=<ltrb 0 0 0 0>
       outlineSettings {
        bUseBrushTransparency=false
        roundingType=HalfHeightRadius
        width=0
        color {colorUseRule=UseColor_Specified, specifiedColor=<rgba 0 0 0 0>}
        cornerRadii=<0 0 0 0>
       }
       tintColor {colorUseRule=UseColor_ColorTable, specifiedColor=<rgba 0 0 0 0>}
       uVRegion {bIsValid=false, max=<0 0>, min=<0 0>}
      }
      selectedBackgroundColor {colorUseRule=UseColor_ColorTable, specifiedColor=<rgba 0 0 0 0>}
      shadowColorAndOpacity=<rgba 0 0 0 1>
      shadowOffset=<0 0>
      strikeBrush {
       bIsDynamicallyLoaded=false
       drawAs=Image
       imageType=NoImage
       mirroring=NoMirror
       resourceName=None
       resourceObject=null
       tiling=NoTile
       imageSize=<32 32>
       margin=<ltrb 0 0 0 0>
       outlineSettings {
        bUseBrushTransparency=false
        roundingType=HalfHeightRadius
        width=0
        color {colorUseRule=UseColor_Specified, specifiedColor=<rgba 0 0 0 0>}
        cornerRadii=<0 0 0 0>
       }
       tintColor {colorUseRule=UseColor_Specified, specifiedColor=<rgba 1 1 1 1>}
       uVRegion {bIsValid=false, max=<0 0>, min=<0 0>}
      }
      underlineBrush {
       bIsDynamicallyLoaded=false
       drawAs=Image
       imageType=NoImage
       mirroring=NoMirror
       resourceName=None
       resourceObject=null
       tiling=NoTile
       imageSize=<32 32>
       margin=<ltrb 0 0 0 0>
       outlineSettings {
        bUseBrushTransparency=false
        roundingType=HalfHeightRadius
        width=0
        color {colorUseRule=UseColor_Specified, specifiedColor=<rgba 0 0 0 0>}
        cornerRadii=<0 0 0 0>
       }
       tintColor {colorUseRule=UseColor_Specified, specifiedColor=<rgba 1 1 1 1>}
       uVRegion {bIsValid=false, max=<0 0>, min=<0 0>}
      }
     }
     vScrollBarPadding=<ltrb 0 0 0 0>
    }
   }
   slot {class=HorizontalBoxSlot, properties {size {sizeRule=Fill, value=1}}}
  }
  Spacer {
   widgetClassPath=&Spacer
   parent=FrameRangeHorizontalBox
   widget=Spacer
   slot {class=HorizontalBoxSlot, properties {size {sizeRule=Fill, value=1}}}
  }
  actionButtons_vert_right {
   widgetClassPath=&VerticalBox
   parent=actionButtons_horiz
   widget=actionButtons_vert_right
   slot {class=HorizontalBoxSlot, properties {verticalAlignment=VAlign_Top, padding=<ltrb 5 5 5 5>, size {sizeRule=Fill, value=1}}}
  }
  Button_Clear_Source {
   widgetClassPath=&EditorUtilityButton
   parent=actionButtons_vert_right
   widget=Button_Clear_Source
   properties {
    clipping=ClipToBounds
    widgetStyle {
     clickedSlateSound {resourceObject=null}
     disabled {
      bIsDynamicallyLoaded=false
      drawAs=RoundedBox
      imageType=NoImage
      mirroring=NoMirror
      resourceName=None
      resourceObject=null
      tiling=NoTile
      imageSize=<0 0>
      margin=<ltrb 0 0 0 0>
      outlineSettings {
       bUseBrushTransparency=false
       roundingType=HalfHeightRadius
       width=1
       color {colorUseRule=UseColor_Specified, specifiedColor=<rgba 0.401 0.401 0.401 0.25>}
       cornerRadii=<4 4 4 4>
      }
      tintColor {colorUseRule=UseColor_Specified, specifiedColor=<rgba 0.0194 0.0194 0.0194 1>}
      uVRegion {bIsValid=false, max=<0 0>, min=<0 0>}
     }
     disabledForeground {colorUseRule=UseColor_ColorTable, specifiedColor=<rgba 0 0 0 0>}
     hovered {
      bIsDynamicallyLoaded=false
      drawAs=RoundedBox
      imageType=NoImage
      mirroring=NoMirror
      resourceName=None
      resourceObject=null
      tiling=NoTile
      imageSize=<32 32>
      margin=<ltrb 0 0 0 0>
      outlineSettings {
       bUseBrushTransparency=false
       roundingType=HalfHeightRadius
       width=1
       color {colorUseRule=UseColor_ColorTable, specifiedColor=<rgba 0 0 0 0>}
       cornerRadii=<4 4 4 4>
      }
      tintColor {colorUseRule=UseColor_Specified, specifiedColor=<rgba 0.3125 0.3125 0.3125 1>}
      uVRegion {bIsValid=false, max=<0 0>, min=<0 0>}
     }
     hoveredForeground {colorUseRule=UseColor_ColorTable, specifiedColor=<rgba 0 0 0 0>}
     hoveredSlateSound {resourceObject=null}
     normal {
      bIsDynamicallyLoaded=false
      drawAs=RoundedBox
      imageType=NoImage
      mirroring=NoMirror
      resourceName=None
      resourceObject=null
      tiling=NoTile
      imageSize=<32 32>
      margin=<ltrb 0 0 0 0>
      outlineSettings {
       bUseBrushTransparency=false
       roundingType=HalfHeightRadius
       width=1
       color {colorUseRule=UseColor_Specified, specifiedColor=<rgba 1 1 1 0.25>}
       cornerRadii=<4 4 4 4>
      }
      tintColor {colorUseRule=UseColor_Specified, specifiedColor=<rgba 0.0194 0.0194 0.0194 1>}
      uVRegion {bIsValid=false, max=<0 0>, min=<0 0>}
     }
     normalForeground {colorUseRule=UseColor_ColorTable, specifiedColor=<rgba 0 0 0 0>}
     normalPadding=<ltrb 12 1.5 12 1.5>
     pressed {
      bIsDynamicallyLoaded=false
      drawAs=RoundedBox
      imageType=NoImage
      mirroring=NoMirror
      resourceName=None
      resourceObject=null
      tiling=NoTile
      imageSize=<32 32>
      margin=<ltrb 0 0 0 0>
      outlineSettings {
       bUseBrushTransparency=false
       roundingType=FixedRadius
       width=1
       color {colorUseRule=UseColor_ColorTable, specifiedColor=<rgba 0 0 0 0>}
       cornerRadii=<4 4 4 4>
      }
      tintColor {colorUseRule=UseColor_Specified, specifiedColor=<rgba 0.0284 0.0284 0.0284 1>}
      uVRegion {bIsValid=false, max=<0 0>, min=<0 0>}
     }
     pressedForeground {colorUseRule=UseColor_ColorTable, specifiedColor=<rgba 0 0 0 0>}
     pressedPadding=<ltrb 12 2.5 12 0.5>
     pressedSlateSound {resourceObject=null}
    }
   }
   slot {class=VerticalBoxSlot, properties {padding=<ltrb 5 5 5 5>, size {sizeRule=Automatic, value=0.1}}}
  }
  TextBlock_Copy_Source_1 {
   widgetClassPath=&TextBlock
   parent=Button_Clear_Source
   widget=TextBlock_Copy_Source_1
   properties {
    text=clear
    font {
     bForceMonospaced=false
     bMaterialIsStencil=false
     fontMaterial=null
     fontObject=&Roboto
     letterSpacing=0
     monospacedWidth=1
     size=18.75
     skewAmount=0
     typefaceFontName=Light
     outlineSettings {
      bApplyOutlineToDropShadows=false
      bMiteredCorners=false
      bSeparateFillAlpha=false
      outlineMaterial=null
      outlineSize=0
      outlineColor=<rgba 0 0 0 1>
     }
    }
   }
   slot {class=ButtonSlot, properties {padding=<ltrb 0 0 0 0>}}
  }
  Button_Paste_Transforms_Rel {
   widgetClassPath=&EditorUtilityButton
   parent=actionButtons_vert_right
   widget=Button_Paste_Transforms_Rel
   properties {
    clipping=ClipToBounds
    widgetStyle {
     clickedSlateSound {resourceObject=null}
     disabled {
      bIsDynamicallyLoaded=false
      drawAs=RoundedBox
      imageType=NoImage
      mirroring=NoMirror
      resourceName=None
      resourceObject=null
      tiling=NoTile
      imageSize=<0 0>
      margin=<ltrb 0 0 0 0>
      outlineSettings {
       bUseBrushTransparency=false
       roundingType=HalfHeightRadius
       width=1
       color {colorUseRule=UseColor_ColorTable, specifiedColor=<rgba 0 0 0 0>}
       cornerRadii=<4 4 4 4>
      }
      tintColor {colorUseRule=UseColor_Specified, specifiedColor=<rgba 0.0395 0.0395 0.0395 1>}
      uVRegion {bIsValid=false, max=<0 0>, min=<0 0>}
     }
     disabledForeground {colorUseRule=UseColor_ColorTable, specifiedColor=<rgba 0 0 0 0>}
     hovered {
      bIsDynamicallyLoaded=false
      drawAs=RoundedBox
      imageType=NoImage
      mirroring=NoMirror
      resourceName=None
      resourceObject=null
      tiling=NoTile
      imageSize=<32 32>
      margin=<ltrb 0 0 0 0>
      outlineSettings {
       bUseBrushTransparency=false
       roundingType=HalfHeightRadius
       width=1
       color {colorUseRule=UseColor_ColorTable, specifiedColor=<rgba 0 0 0 0>}
       cornerRadii=<4 4 4 4>
      }
      tintColor {colorUseRule=UseColor_Specified, specifiedColor=<rgba 0.0953 0.0953 0.0953 1>}
      uVRegion {bIsValid=false, max=<0 0>, min=<0 0>}
     }
     hoveredForeground {colorUseRule=UseColor_ColorTable, specifiedColor=<rgba 0 0 0 0>}
     hoveredSlateSound {resourceObject=null}
     normal {
      bIsDynamicallyLoaded=false
      drawAs=RoundedBox
      imageType=NoImage
      mirroring=NoMirror
      resourceName=None
      resourceObject=null
      tiling=NoTile
      imageSize=<32 32>
      margin=<ltrb 0 0 0 0>
      outlineSettings {
       bUseBrushTransparency=false
       roundingType=HalfHeightRadius
       width=1
       color {colorUseRule=UseColor_ColorTable, specifiedColor=<rgba 0 0 0 0>}
       cornerRadii=<4 4 4 4>
      }
      tintColor {colorUseRule=UseColor_Specified, specifiedColor=<rgba 0.0741 0.1458 0.0706 1>}
      uVRegion {bIsValid=false, max=<0 0>, min=<0 0>}
     }
     normalForeground {colorUseRule=UseColor_ColorTable, specifiedColor=<rgba 0 0 0 0>}
     normalPadding=<ltrb 12 1.5 12 1.5>
     pressed {
      bIsDynamicallyLoaded=false
      drawAs=RoundedBox
      imageType=NoImage
      mirroring=NoMirror
      resourceName=None
      resourceObject=null
      tiling=NoTile
      imageSize=<32 32>
      margin=<ltrb 0 0 0 0>
      outlineSettings {
       bUseBrushTransparency=false
       roundingType=FixedRadius
       width=1
       color {colorUseRule=UseColor_ColorTable, specifiedColor=<rgba 0 0 0 0>}
       cornerRadii=<4 4 4 4>
      }
      tintColor {colorUseRule=UseColor_Specified, specifiedColor=<rgba 0.0284 0.0284 0.0284 1>}
      uVRegion {bIsValid=false, max=<0 0>, min=<0 0>}
     }
     pressedForeground {colorUseRule=UseColor_ColorTable, specifiedColor=<rgba 0 0 0 0>}
     pressedPadding=<ltrb 12 2.5 12 0.5>
     pressedSlateSound {resourceObject=null}
    }
   }
   slot {class=VerticalBoxSlot, properties {padding=<ltrb 5 1 5 5>}}
  }
  TextBlock_Paste_Transforms_2 {
   widgetClassPath=&TextBlock
   parent=Button_Paste_Transforms_Rel
   widget=TextBlock_Paste_Transforms_2
   properties {
    text=Paste Relative
    font {
     bForceMonospaced=false
     bMaterialIsStencil=false
     fontMaterial=null
     fontObject=&Roboto
     letterSpacing=0
     monospacedWidth=1
     size=22.5
     skewAmount=0
     typefaceFontName=Bold
     outlineSettings {
      bApplyOutlineToDropShadows=false
      bMiteredCorners=false
      bSeparateFillAlpha=false
      outlineMaterial=null
      outlineSize=0
      outlineColor=<rgba 0 0 0 1>
     }
    }
   }
   slot {class=ButtonSlot, properties {verticalAlignment=VAlign_Fill, padding=<ltrb 0 0 0 0>}}
  }
  Button_Paste_Transforms {
   widgetClassPath=&EditorUtilityButton
   parent=actionButtons_vert_right
   widget=Button_Paste_Transforms
   properties {
    clipping=ClipToBounds
    widgetStyle {
     clickedSlateSound {resourceObject=null}
     disabled {
      bIsDynamicallyLoaded=false
      drawAs=RoundedBox
      imageType=NoImage
      mirroring=NoMirror
      resourceName=None
      resourceObject=null
      tiling=NoTile
      imageSize=<0 0>
      margin=<ltrb 0 0 0 0>
      outlineSettings {
       bUseBrushTransparency=false
       roundingType=HalfHeightRadius
       width=1
       color {colorUseRule=UseColor_ColorTable, specifiedColor=<rgba 0 0 0 0>}
       cornerRadii=<4 4 4 4>
      }
      tintColor {colorUseRule=UseColor_Specified, specifiedColor=<rgba 0.0395 0.0395 0.0395 1>}
      uVRegion {bIsValid=false, max=<0 0>, min=<0 0>}
     }
     disabledForeground {colorUseRule=UseColor_ColorTable, specifiedColor=<rgba 0 0 0 0>}
     hovered {
      bIsDynamicallyLoaded=false
      drawAs=RoundedBox
      imageType=NoImage
      mirroring=NoMirror
      resourceName=None
      resourceObject=null
      tiling=NoTile
      imageSize=<32 32>
      margin=<ltrb 0 0 0 0>
      outlineSettings {
       bUseBrushTransparency=false
       roundingType=HalfHeightRadius
       width=1
       color {colorUseRule=UseColor_ColorTable, specifiedColor=<rgba 0 0 0 0>}
       cornerRadii=<4 4 4 4>
      }
      tintColor {colorUseRule=UseColor_Specified, specifiedColor=<rgba 0.0953 0.0953 0.0953 1>}
      uVRegion {bIsValid=false, max=<0 0>, min=<0 0>}
     }
     hoveredForeground {colorUseRule=UseColor_ColorTable, specifiedColor=<rgba 0 0 0 0>}
     hoveredSlateSound {resourceObject=null}
     normal {
      bIsDynamicallyLoaded=false
      drawAs=RoundedBox
      imageType=NoImage
      mirroring=NoMirror
      resourceName=None
      resourceObject=null
      tiling=NoTile
      imageSize=<32 32>
      margin=<ltrb 0 0 0 0>
      outlineSettings {
       bUseBrushTransparency=false
       roundingType=HalfHeightRadius
       width=1
       color {colorUseRule=UseColor_ColorTable, specifiedColor=<rgba 0 0 0 0>}
       cornerRadii=<4 4 4 4>
      }
      tintColor {colorUseRule=UseColor_Specified, specifiedColor=<rgba 0.0741 0.1458 0.0706 1>}
      uVRegion {bIsValid=false, max=<0 0>, min=<0 0>}
     }
     normalForeground {colorUseRule=UseColor_ColorTable, specifiedColor=<rgba 0 0 0 0>}
     normalPadding=<ltrb 12 1.5 12 1.5>
     pressed {
      bIsDynamicallyLoaded=false
      drawAs=RoundedBox
      imageType=NoImage
      mirroring=NoMirror
      resourceName=None
      resourceObject=null
      tiling=NoTile
      imageSize=<32 32>
      margin=<ltrb 0 0 0 0>
      outlineSettings {
       bUseBrushTransparency=false
       roundingType=FixedRadius
       width=1
       color {colorUseRule=UseColor_ColorTable, specifiedColor=<rgba 0 0 0 0>}
       cornerRadii=<4 4 4 4>
      }
      tintColor {colorUseRule=UseColor_Specified, specifiedColor=<rgba 0.0284 0.0284 0.0284 1>}
      uVRegion {bIsValid=false, max=<0 0>, min=<0 0>}
     }
     pressedForeground {colorUseRule=UseColor_ColorTable, specifiedColor=<rgba 0 0 0 0>}
     pressedPadding=<ltrb 12 2.5 12 0.5>
     pressedSlateSound {resourceObject=null}
    }
   }
   slot {class=VerticalBoxSlot}
  }
  TextBlock_Paste_Transforms {
   widgetClassPath=&TextBlock
   parent=Button_Paste_Transforms
   widget=TextBlock_Paste_Transforms
   properties {
    text=Snap to Parent
    font {
     bForceMonospaced=false
     bMaterialIsStencil=false
     fontMaterial=null
     fontObject=&Roboto
     letterSpacing=0
     monospacedWidth=1
     size=22.5
     skewAmount=0
     typefaceFontName=Bold
     outlineSettings {
      bApplyOutlineToDropShadows=false
      bMiteredCorners=false
      bSeparateFillAlpha=false
      outlineMaterial=null
      outlineSize=0
      outlineColor=<rgba 0 0 0 1>
     }
    }
   }
   slot {class=ButtonSlot, properties {verticalAlignment=VAlign_Fill, padding=<ltrb 0 0 0 0>}}
  }
 ]
}
