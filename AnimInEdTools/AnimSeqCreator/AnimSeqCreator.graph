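# Editor Utility Widget front end for creating animation sequences, with frame count, additive, relative-to-frame-0, asset name and output directory fields.
# On Construct, an Execute Python Script node reloads the AnimSequenceCreator module and hands it the widgets and owner.
# Button behavior, including skeletal mesh loading and creation, is left to that Python code.
# The script also references text_block_loaded_mesh, which has no matching widget or input pin here.
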EditorUtilityWidgetBlueprint /AnimInEdTools/AnimSeqCreator/AnimSeqCreator
graphs [EventGraph {nodes [
   Event Construct()
   > Execute Python Script!(
    execute
    PythonScript="import importlib, AnimSequenceCreator; importlib.reload(AnimSequenceCreator); tool = AnimSequenceCreator.cls;\r\ntool.init_owner(owner); tool.init_widget_editable_frame_count(editable_text_frame_count);\r\ntool.init_widget_checkbox_additive(check_box_additive);\r\ntool.init_widget_checkbox_relative_to_zero(check_box_relative_to_frame0);\r\ntool.init_widget_editable_output_path(editable_text_output_path);\r\ntool.init_widget_editable_asset_name(editable_text_asset_name); tool.init_widget_button_create(button_create);\r\ntool.init_widget_text_status(text_block_status); tool.init_widget_button_load_selected(button_load_selected);\r\ntool.init_widget_text_loaded_mesh(text_block_loaded_mesh)"
    CheckBox_Additive:Editor Utility Check Box Object Reference=Get CheckBox_Additive() -> (CheckBox_Additive:Editor Utility Check Box Object Reference)
    CheckBox_RelativeToFrame0:Editor Utility Check Box Object Reference=Get CheckBox_RelativeToFrame0() -> (CheckBox_RelativeToFrame0:Editor Utility Check Box Object Reference)
    EditableText_OutputPath:Editor Utility Editable Text Box Object Reference=Get EditableText_OutputPath() -> (EditableText_OutputPath:Editor Utility Editable Text Box Object Reference)
    Button_Create:Editor Utility Button Object Reference=Get Button_Create() -> (Button_Create:Editor Utility Button Object Reference)
    TextBlock_Status:Text Object Reference=Get TextBlock_Status() -> (TextBlock_Status:Text Object Reference)
    Button_LoadSelected:Editor Utility Button Object Reference=Get Button_LoadSelected() -> (Button_LoadSelected:Editor Utility Button Object Reference)
    Owner:Self Object Reference=Self-Reference()
    EditableText_FrameCount:Editor Utility Editable Text Box Object Reference=Get EditableText_FrameCount() -> (EditableText_FrameCount:Editor Utility Editable Text Box Object Reference)
    EditableText_AssetName:Editor Utility Editable Text Box Object Reference=Get EditableText_AssetName() -> (EditableText_AssetName:Editor Utility Editable Text Box Object Reference)
   ) -> (then, ReturnValue) {inputs [
     CheckBox_Additive
     CheckBox_RelativeToFrame0
     EditableText_OutputPath
     Button_Create
     TextBlock_Status
     Button_LoadSelected
     Owner
     EditableText_FrameCount
     EditableText_AssetName
    ]}
   Event PreConstruct()
   Event Tick()
  ]}]
variables [SkeletalMesh {type=SkeletalMesh, default=null}]
widgets {
 info {rootWidgetClass=&StackBox, widgetCount=22}
 widgets [
  StackBox_0 {widgetClassPath=&StackBox, widget=StackBox_0}
  VerticalBox_55 {
   widgetClassPath=&VerticalBox
   parent=StackBox_0
   widget=VerticalBox_55
   slot {class=StackBoxSlot, properties {size {sizeRule=Fill, value=1}}}
  }
  TextBlock_363 {
   widgetClassPath=&TextBlock
   parent=VerticalBox_55
   widget=TextBlock_363
   properties {
    text="ANIM SEQ CREATOR "
    colorAndOpacity {colorUseRule=UseColor_Specified, specifiedColor=<rgba 1 0.4414 0 1>}
    font {
     bForceMonospaced=false
     bMaterialIsStencil=false
     fontMaterial=null
     fontObject=null
     letterSpacing=0
     monospacedWidth=1
     size=18.75
     skewAmount=0
     typefaceFontName=Default
     outlineSettings {
      bApplyOutlineToDropShadows=false
      bMiteredCorners=false
      bSeparateFillAlpha=false
      outlineMaterial=null
      outlineSize=0
      outlineColor=<rgba 0 0 0 1>
     }
    }
   }
   slot {class=VerticalBoxSlot, properties {horizontalAlignment=HAlign_Center}}
  }
  HorizontalBox_406 {
   widgetClassPath=&HorizontalBox
   parent=VerticalBox_55
   widget=HorizontalBox_406
   slot {class=VerticalBoxSlot, properties {horizontalAlignment=HAlign_Center}}
  }
  TextBlock_432 {
   widgetClassPath=&TextBlock
   parent=HorizontalBox_406
   widget=TextBlock_432
   properties {
    text=Frame Count:
    font {
     bForceMonospaced=false
     bMaterialIsStencil=false
     fontMaterial=null
     fontObject=null
     letterSpacing=0
     monospacedWidth=1
     size=24
     skewAmount=0
     typefaceFontName=Default
     outlineSettings {
      bApplyOutlineToDropShadows=false
      bMiteredCorners=false
      bSeparateFillAlpha=false
      outlineMaterial=null
      outlineSize=0
      outlineColor=<rgba 0 0 0 1>
     }
    }
   }
   slot {class=HorizontalBoxSlot, properties {horizontalAlignment=HAlign_Center}}
  }
  EditableText_FrameCount {
   widgetClassPath=&EditorUtilityEditableTextBox
   bIsVariable=true
   parent=HorizontalBox_406
   widget=EditableText_FrameCount
   slot {class=HorizontalBoxSlot, properties {horizontalAlignment=HAlign_Center}}
  }
  HorizontalBox_185 {
   widgetClassPath=&HorizontalBox
   parent=VerticalBox_55
   widget=HorizontalBox_185
   slot {class=VerticalBoxSlot, properties {horizontalAlignment=HAlign_Center}}
  }
  TextBlock_518 {
   widgetClassPath=&TextBlock
   parent=HorizontalBox_185
   widget=TextBlock_518
   properties {
    text="Additive: "
    font {
     bForceMonospaced=false
     bMaterialIsStencil=false
     fontMaterial=null
     fontObject=null
     letterSpacing=0
     monospacedWidth=1
     size=24
     skewAmount=0
     typefaceFontName=Default
     outlineSettings {
      bApplyOutlineToDropShadows=false
      bMiteredCorners=false
      bSeparateFillAlpha=false
      outlineMaterial=null
      outlineSize=0
      outlineColor=<rgba 0 0 0 1>
     }
    }
   }
   slot {class=HorizontalBoxSlot}
  }
  CheckBox_Additive {
   widgetClassPath=&EditorUtilityCheckBox
   bIsVariable=true
   parent=HorizontalBox_185
   widget=CheckBox_Additive
   slot {class=HorizontalBoxSlot, properties {size {sizeRule=Fill, value=5}}}
  }
  TextBlock {
   widgetClassPath=&TextBlock
   parent=HorizontalBox_185
   widget=TextBlock
   properties {
    text="Relative to Frame 0: "
    font {
     bForceMonospaced=false
     bMaterialIsStencil=false
     fontMaterial=null
     fontObject=null
     letterSpacing=0
     monospacedWidth=1
     size=24
     skewAmount=0
     typefaceFontName=Default
     outlineSettings {
      bApplyOutlineToDropShadows=false
      bMiteredCorners=false
      bSeparateFillAlpha=false
      outlineMaterial=null
      outlineSize=0
      outlineColor=<rgba 0 0 0 1>
     }
    }
   }
   slot {class=HorizontalBoxSlot}
  }
  CheckBox_RelativeToFrame0 {
   widgetClassPath=&EditorUtilityCheckBox
   bIsVariable=true
   parent=HorizontalBox_185
   widget=CheckBox_RelativeToFrame0
   slot {class=HorizontalBoxSlot, properties {size {sizeRule=Fill, value=5}}}
  }
  HorizontalBox_852 {
   widgetClassPath=&HorizontalBox
   parent=VerticalBox_55
   widget=HorizontalBox_852
   slot {class=VerticalBoxSlot, properties {horizontalAlignment=HAlign_Center}}
  }
  TextBlock_1 {
   widgetClassPath=&TextBlock
   parent=HorizontalBox_852
   widget=TextBlock_1
   properties {
    text=Asset Name:
    font {
     bForceMonospaced=false
     bMaterialIsStencil=false
     fontMaterial=null
     fontObject=null
     letterSpacing=0
     monospacedWidth=1
     size=24
     skewAmount=0
     typefaceFontName=Default
     outlineSettings {
      bApplyOutlineToDropShadows=false
      bMiteredCorners=false
      bSeparateFillAlpha=false
      outlineMaterial=null
      outlineSize=0
      outlineColor=<rgba 0 0 0 1>
     }
    }
   }
   slot {class=HorizontalBoxSlot, properties {horizontalAlignment=HAlign_Center}}
  }
  EditableText_AssetName {
   widgetClassPath=&EditorUtilityEditableTextBox
   bIsVariable=true
   parent=HorizontalBox_852
   widget=EditableText_AssetName
   slot {class=HorizontalBoxSlot}
  }
  HorizontalBox_967 {
   widgetClassPath=&HorizontalBox
   parent=VerticalBox_55
   widget=HorizontalBox_967
   slot {class=VerticalBoxSlot, properties {horizontalAlignment=HAlign_Center}}
  }
  TextBlock_2 {
   widgetClassPath=&TextBlock
   parent=HorizontalBox_967
   widget=TextBlock_2
   properties {
    text=Output Directory:
    font {
     bForceMonospaced=false
     bMaterialIsStencil=false
     fontMaterial=null
     fontObject=null
     letterSpacing=0
     monospacedWidth=1
     size=24
     skewAmount=0
     typefaceFontName=Default
     outlineSettings {
      bApplyOutlineToDropShadows=false
      bMiteredCorners=false
      bSeparateFillAlpha=false
      outlineMaterial=null
      outlineSize=0
      outlineColor=<rgba 0 0 0 1>
     }
    }
   }
   slot {class=HorizontalBoxSlot, properties {horizontalAlignment=HAlign_Center}}
  }
  EditableText_OutputPath {
   widgetClassPath=&EditorUtilityEditableTextBox
   bIsVariable=true
   parent=HorizontalBox_967
   widget=EditableText_OutputPath
   slot {class=HorizontalBoxSlot}
  }
  Button_LoadSelected {
   widgetClassPath=&EditorUtilityButton
   parent=VerticalBox_55
   widget=Button_LoadSelected
   properties {widgetStyle {
     clickedSlateSound {resourceObject=null}
     disabled {
      bIsDynamicallyLoaded=false
      drawAs=RoundedBox
      imageType=NoImage
      mirroring=NoMirror
      resourceName=None
      resourceObject=null
      tiling=NoTile
      imageSize=<0 0>
      margin=<ltrb 0 0 0 0>
      outlineSettings {
       bUseBrushTransparency=false
       roundingType=FixedRadius
       width=1
       color {colorUseRule=UseColor_ColorTable, specifiedColor=<rgba 0 0 0 0>}
       cornerRadii=<4 4 4 4>
      }
      tintColor {colorUseRule=UseColor_Specified, specifiedColor=<rgba 0.0395 0.0395 0.0395 1>}
      uVRegion {bIsValid=false, max=<0 0>, min=<0 0>}
     }
     disabledForeground {colorUseRule=UseColor_ColorTable, specifiedColor=<rgba 0 0 0 0>}
     hovered {
      bIsDynamicallyLoaded=false
      drawAs=RoundedBox
      imageType=NoImage
      mirroring=NoMirror
      resourceName=None
      resourceObject=null
      tiling=NoTile
      imageSize=<32 32>
      margin=<ltrb 0 0 0 0>
      outlineSettings {
       bUseBrushTransparency=false
       roundingType=FixedRadius
       width=1
       color {colorUseRule=UseColor_ColorTable, specifiedColor=<rgba 0 0 0 0>}
       cornerRadii=<4 4 4 4>
      }
      tintColor {colorUseRule=UseColor_Specified, specifiedColor=<rgba 0.0953 0.0953 0.0953 1>}
      uVRegion {bIsValid=false, max=<0 0>, min=<0 0>}
     }
     hoveredForeground {colorUseRule=UseColor_ColorTable, specifiedColor=<rgba 0 0 0 0>}
     hoveredSlateSound {resourceObject=null}
     normal {
      bIsDynamicallyLoaded=false
      drawAs=RoundedBox
      imageType=NoImage
      mirroring=NoMirror
      resourceName=None
      resourceObject=null
      tiling=NoTile
      imageSize=<32 32>
      margin=<ltrb 0 0 0 0>
      outlineSettings {
       bUseBrushTransparency=false
       roundingType=FixedRadius
       width=1
       color {colorUseRule=UseColor_ColorTable, specifiedColor=<rgba 0 0 0 0>}
       cornerRadii=<4 4 4 4>
      }
      tintColor {colorUseRule=UseColor_Specified, specifiedColor=<rgba 0.0585 0.401 0.3385 1>}
      uVRegion {bIsValid=false, max=<0 0>, min=<0 0>}
     }
     normalForeground {colorUseRule=UseColor_ColorTable, specifiedColor=<rgba 0 0 0 0>}
     normalPadding=<ltrb 12 1.5 12 1.5>
     pressed {
      bIsDynamicallyLoaded=false
      drawAs=RoundedBox
      imageType=NoImage
      mirroring=NoMirror
      resourceName=None
      resourceObject=null
      tiling=NoTile
      imageSize=<32 32>
      margin=<ltrb 0 0 0 0>
      outlineSettings {
       bUseBrushTransparency=false
       roundingType=FixedRadius
       width=1
       color {colorUseRule=UseColor_ColorTable, specifiedColor=<rgba 0 0 0 0>}
       cornerRadii=<4 4 4 4>
      }
      tintColor {colorUseRule=UseColor_Specified, specifiedColor=<rgba 0.0284 0.0284 0.0284 1>}
      uVRegion {bIsValid=false, max=<0 0>, min=<0 0>}
     }
     pressedForeground {colorUseRule=UseColor_ColorTable, specifiedColor=<rgba 0 0 0 0>}
     pressedPadding=<ltrb 12 2.5 12 0.5>
     pressedSlateSound {resourceObject=null}
    }}
   slot {class=VerticalBoxSlot, properties {padding=<ltrb 3 3 3 3>}}
  }
  TextBlock_105 {
   widgetClassPath=&TextBlock
   parent=Button_LoadSelected
   widget=TextBlock_105
   properties {
    text=LOAD SKM
    font {
     bForceMonospaced=false
     bMaterialIsStencil=false
     fontMaterial=null
     fontObject=null
     letterSpacing=0
     monospacedWidth=1
     size=24
     skewAmount=0
     typefaceFontName=Default
     outlineSettings {
      bApplyOutlineToDropShadows=false
      bMiteredCorners=false
      bSeparateFillAlpha=false
      outlineMaterial=null
      outlineSize=0
      outlineColor=<rgba 0 0 0 1>
     }
    }
   }
   slot {class=ButtonSlot}
  }
  Button_Create {
   widgetClassPath=&EditorUtilityButton
   parent=VerticalBox_55
   widget=Button_Create
   properties {widgetStyle {
     clickedSlateSound {resourceObject=null}
     disabled {
      bIsDynamicallyLoaded=false
      drawAs=RoundedBox
      imageType=NoImage
      mirroring=NoMirror
      resourceName=None
      resourceObject=null
      tiling=NoTile
      imageSize=<0 0>
      margin=<ltrb 0 0 0 0>
      outlineSettings {
       bUseBrushTransparency=false
       roundingType=FixedRadius
       width=1
       color {colorUseRule=UseColor_ColorTable, specifiedColor=<rgba 0 0 0 0>}
       cornerRadii=<4 4 4 4>
      }
      tintColor {colorUseRule=UseColor_Specified, specifiedColor=<rgba 0.0395 0.0395 0.0395 1>}
      uVRegion {bIsValid=false, max=<0 0>, min=<0 0>}
     }
     disabledForeground {colorUseRule=UseColor_ColorTable, specifiedColor=<rgba 0 0 0 0>}
     hovered {
      bIsDynamicallyLoaded=false
      drawAs=RoundedBox
      imageType=NoImage
      mirroring=NoMirror
      resourceName=None
      resourceObject=null
      tiling=NoTile
      imageSize=<32 32>
      margin=<ltrb 0 0 0 0>
      outlineSettings {
       bUseBrushTransparency=false
       roundingType=FixedRadius
       width=1
       color {colorUseRule=UseColor_ColorTable, specifiedColor=<rgba 0 0 0 0>}
       cornerRadii=<4 4 4 4>
      }
      tintColor {colorUseRule=UseColor_Specified, specifiedColor=<rgba 0.0953 0.0953 0.0953 1>}
      uVRegion {bIsValid=false, max=<0 0>, min=<0 0>}
     }
     hoveredForeground {colorUseRule=UseColor_ColorTable, specifiedColor=<rgba 0 0 0 0>}
     hoveredSlateSound {resourceObject=null}
     normal {
      bIsDynamicallyLoaded=false
      drawAs=RoundedBox
      imageType=NoImage
      mirroring=NoMirror
      resourceName=None
      resourceObject=null
      tiling=NoTile
      imageSize=<32 32>
      margin=<ltrb 0 0 0 0>
      outlineSettings {
       bUseBrushTransparency=false
       roundingType=FixedRadius
       width=1
       color {colorUseRule=UseColor_ColorTable, specifiedColor=<rgba 0 0 0 0>}
       cornerRadii=<4 4 4 4>
      }
      tintColor {colorUseRule=UseColor_Specified, specifiedColor=<rgba 0.0741 0.401 0.0585 1>}
      uVRegion {bIsValid=false, max=<0 0>, min=<0 0>}
     }
     normalForeground {colorUseRule=UseColor_ColorTable, specifiedColor=<rgba 0 0 0 0>}
     normalPadding=<ltrb 12 1.5 12 1.5>
     pressed {
      bIsDynamicallyLoaded=false
      drawAs=RoundedBox
      imageType=NoImage
      mirroring=NoMirror
      resourceName=None
      resourceObject=null
      tiling=NoTile
      imageSize=<32 32>
      margin=<ltrb 0 0 0 0>
      outlineSettings {
       bUseBrushTransparency=false
       roundingType=FixedRadius
       width=1
       color {colorUseRule=UseColor_ColorTable, specifiedColor=<rgba 0 0 0 0>}
       cornerRadii=<4 4 4 4>
      }
      tintColor {colorUseRule=UseColor_Specified, specifiedColor=<rgba 0.0284 0.0284 0.0284 1>}
      uVRegion {bIsValid=false, max=<0 0>, min=<0 0>}
     }
     pressedForeground {colorUseRule=UseColor_ColorTable, specifiedColor=<rgba 0 0 0 0>}
     pressedPadding=<ltrb 12 2.5 12 0.5>
     pressedSlateSound {resourceObject=null}
    }}
   slot {class=VerticalBoxSlot, properties {padding=<ltrb 3 3 3 3>}}
  }
  TextBlock_103 {
   widgetClassPath=&TextBlock
   parent=Button_Create
   widget=TextBlock_103
   properties {
    text="CREATE ANIM SEQ!"
    font {
     bForceMonospaced=false
     bMaterialIsStencil=false
     fontMaterial=null
     fontObject=null
     letterSpacing=0
     monospacedWidth=1
     size=24
     skewAmount=0
     typefaceFontName=Default
     outlineSettings {
      bApplyOutlineToDropShadows=false
      bMiteredCorners=false
      bSeparateFillAlpha=false
      outlineMaterial=null
      outlineSize=0
      outlineColor=<rgba 0 0 0 1>
     }
    }
   }
   slot {class=ButtonSlot}
  }
  TextBlock_Status {
   widgetClassPath=&TextBlock
   bIsVariable=true
   parent=VerticalBox_55
   widget=TextBlock_Status
   properties {
    text=Status
    font {
     bForceMonospaced=false
     bMaterialIsStencil=false
     fontMaterial=null
     fontObject=null
     letterSpacing=0
     monospacedWidth=1
     size=24
     skewAmount=0
     typefaceFontName=Default
     outlineSettings {
      bApplyOutlineToDropShadows=false
      bMiteredCorners=false
      bSeparateFillAlpha=false
      outlineMaterial=null
      outlineSize=0
      outlineColor=<rgba 0 0 0 1>
     }
    }
   }
   slot {class=VerticalBoxSlot, properties {horizontalAlignment=HAlign_Center}}
  }
 ]
}
