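; AnimatableCreature supplies an animatable creature's file-naming options (anim sets, descriptors, directions, variations, subfolders) for the AnimInEdTools animation browser.
; Getters synchronously load linked Project assets and combine their names with the creature's FileNameBase assets via GetFileNamesForProject, where a non-empty override array replaces both.
; DataAsset getters overwrite their list per project, keeping only the last project's assets before adding the creature's own.
; GetNamingConvention lazily constructs and caches its convention object.
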(fn GetAnimSet
  (Utilities|Array|Clear (LocalVariables|GetNames%20Local))
  (for _array_element (Variables|Project|GetProjects)
    (bind _returnvalue (Utilities|LoadAssetBlocking _array_element))
    (bind _asbp_animation_browser_project (Utilities|Casting|CastToBP_AnimationBrowser_Project _returnvalue))
    (Utilities|Array|AppendArray (LocalVariables|GetNames%20Local) (FileNaming|GetAnimSets _asbp_animation_browser_project)))
  (return (Private|GetFileNamesforProject :Append%20Array (Variables|Project|GetAnim%20Sets) :Project%20Names (LocalVariables|GetNames%20Local))))

(fn GetAnimSetsDataAssets
  (Utilities|Array|Clear (LocalVariables|GetData%20Assets%20Local))
  (for _array_element (Variables|Project|GetProjects)
    (bind _returnvalue (Utilities|LoadAssetBlocking _array_element))
    (bind _asbp_animation_browser_project (Utilities|Casting|CastToBP_AnimationBrowser_Project _returnvalue))
    (bind _data_assets (DataAssets|GetAnimSetsDataAsset _asbp_animation_browser_project))
    (LocalVariables|SetData%20Assets%20Local _data_assets))
  (for _array_element_2 (Variables|Project|GetAnim%20Sets)
    (bind _returnvalue_2 (Utilities|LoadAssetBlocking _array_element_2))
    (bind _asbp_animation_browser_file_name_base (Utilities|Casting|CastToBP_AnimationBrowser_FileNameBase _returnvalue_2))
    (Utilities|Array|Add (LocalVariables|GetData%20Assets%20Local) _asbp_animation_browser_file_name_base))
  (return (LocalVariables|GetData%20Assets%20Local)))

(fn GetDescriptorDataAssets
  (Utilities|Array|Clear (LocalVariables|GetData%20Assets%20Local))
  (for _array_element (Variables|Project|GetProjects)
    (bind _returnvalue (Utilities|LoadAssetBlocking _array_element))
    (bind _asbp_animation_browser_project (Utilities|Casting|CastToBP_AnimationBrowser_Project _returnvalue))
    (bind _data_assets (DataAssets|GetDescriptorsDataAsset _asbp_animation_browser_project))
    (LocalVariables|SetData%20Assets%20Local _data_assets))
  (for _array_element_2 (Variables|Project|GetDescriptors)
    (bind _returnvalue_2 (Utilities|LoadAssetBlocking _array_element_2))
    (bind _asbp_animation_browser_file_name_base (Utilities|Casting|CastToBP_AnimationBrowser_FileNameBase _returnvalue_2))
    (Utilities|Array|Add (LocalVariables|GetData%20Assets%20Local) _asbp_animation_browser_file_name_base))
  (return (LocalVariables|GetData%20Assets%20Local)))

(fn GetDescriptors
  (Utilities|Array|Clear (LocalVariables|GetNames%20Local))
  (for _array_element (Variables|Project|GetProjects)
    (bind _returnvalue (Utilities|LoadAssetBlocking _array_element))
    (bind _asbp_animation_browser_project (Utilities|Casting|CastToBP_AnimationBrowser_Project _returnvalue))
    (Utilities|Array|AppendArray (LocalVariables|GetNames%20Local) (FileNaming|GetDescriptors _asbp_animation_browser_project :!class /AnimInEdTools/AnimationBrowser/ProjectSettings/BP_AnimationBrowser_Project.BP_AnimationBrowser_Project_C)))
  (return (Private|GetFileNamesforProject :Append%20Array (Variables|Project|GetDescriptors) :Project%20Names (LocalVariables|GetNames%20Local))))

(fn GetDirection
  (Utilities|Array|Clear (LocalVariables|GetNames%20Local))
  (for _array_element (Variables|Project|GetProjects)
    (bind _returnvalue (Utilities|LoadAssetBlocking _array_element))
    (bind _asbp_animation_browser_project (Utilities|Casting|CastToBP_AnimationBrowser_Project _returnvalue))
    (Utilities|Array|AppendArray (LocalVariables|GetNames%20Local) (FileNaming|GetDirection _asbp_animation_browser_project :!class /AnimInEdTools/AnimationBrowser/ProjectSettings/BP_AnimationBrowser_Project.BP_AnimationBrowser_Project_C)))
  (return (Private|GetFileNamesforProject :Append%20Array (Variables|Project|GetDirections) :Project%20Names (LocalVariables|GetNames%20Local))))

(fn GetDirectionDataAssets
  (Utilities|Array|Clear (LocalVariables|GetData%20Assets%20Local))
  (for _array_element (Variables|Project|GetProjects)
    (bind _returnvalue (Utilities|LoadAssetBlocking _array_element))
    (bind _asbp_animation_browser_project (Utilities|Casting|CastToBP_AnimationBrowser_Project _returnvalue))
    (bind _data_assets (DataAssets|GetDirectionDataAsset _asbp_animation_browser_project))
    (LocalVariables|SetData%20Assets%20Local _data_assets))
  (for _array_element_2 (Variables|Project|GetDirections)
    (bind _returnvalue_2 (Utilities|LoadAssetBlocking _array_element_2))
    (bind _asbp_animation_browser_file_name_base (Utilities|Casting|CastToBP_AnimationBrowser_FileNameBase _returnvalue_2))
    (Utilities|Array|Add (LocalVariables|GetData%20Assets%20Local) _asbp_animation_browser_file_name_base))
  (return (LocalVariables|GetData%20Assets%20Local)))

(fn GetFileNamesForProject
  (params (Override%20Array :type Array%20of%20BP%20Animation%20Browser%20File%20Name%20Base%20Soft%20Object%20References) (Append%20Array :type Array%20of%20BP%20Animation%20Browser%20File%20Name%20Base%20Soft%20Object%20References) (Project%20Names :type Array%20of%20Strings))
  (Utilities|Array|Clear (LocalVariables|GetNames%20Local))
  (if (Utilities|Array|IsNotEmpty Override%20Array)
    (return (Private|GetFileNameBase (Variables|GetOverride%20Array)))
    (else
      (LocalVariables|SetNames%20Local (Variables|GetProject%20Names))
      (Utilities|Array|AppendArray (LocalVariables|GetNames%20Local) (Private|GetFileNameBase (Variables|GetAppend%20Array)))
      (return (LocalVariables|GetNames%20Local)))))

(fn GetFileNameBase
  (params (Array :type Array%20of%20BP%20Animation%20Browser%20File%20Name%20Base%20Soft%20Object%20References))
  (Utilities|Array|Clear (LocalVariables|GetNames))
  (for _array_element Array
    (bind _returnvalue (Utilities|LoadAssetBlocking _array_element))
    (bind _asbp_animation_browser_file_name_base (Utilities|Casting|CastToBP_AnimationBrowser_FileNameBase _returnvalue))
    (for _array_element_2 (Variables|/AnimInEdTools/AnimationBrowser/ProjectSettings/Data/BP_AnimationBrowser_FileNameBase.BP_AnimationBrowser_FileNameBase_C|GetNames _asbp_animation_browser_file_name_base)
      (Utilities|Array|AddUnique (LocalVariables|GetNames) _array_element_2)))
  (return (LocalVariables|GetNames)))

(fn GetNamingConvention
  (Utilities|IsValid (Variables|Internal|GetNaming%20Convention%20Instance)
    (:Is%20Valid
      (return (Variables|Internal|GetNaming%20Convention%20Instance)))
    (:Is%20Not%20Valid
      (bind _returnvalue (Game|ConstructObjectfromClass (Variables|Naming|GetNaming%20Convention)))
      (bind _output_get (Variables|Internal|SetNaming%20Convention%20Instance _returnvalue))
      (return _output_get))))

(fn GetProjects
  (Utilities|Array|Clear (LocalVariables|GetLocal%20Projects))
  (for _array_element (Variables|Project|GetProjects)
    (bind _returnvalue (Utilities|LoadAssetBlocking _array_element))
    (bind _asbp_animation_browser_project (Utilities|Casting|CastToBP_AnimationBrowser_Project _returnvalue))
    (Utilities|Array|Add (LocalVariables|GetLocal%20Projects) _asbp_animation_browser_project))
  (return (LocalVariables|GetLocal%20Projects)))

(fn GetSubfolders
  (Utilities|Array|Clear (LocalVariables|GetNames%20Local))
  (for _array_element (Variables|Project|GetProjects)
    (bind _returnvalue (Utilities|LoadAssetBlocking _array_element))
    (bind _asbp_animation_browser_project (Utilities|Casting|CastToBP_AnimationBrowser_Project _returnvalue))
    (for _array_element_2 (FileNaming|GetSubfolders _asbp_animation_browser_project :!class /AnimInEdTools/AnimationBrowser/ProjectSettings/BP_AnimationBrowser_Project.BP_AnimationBrowser_Project_C)
      (Utilities|Array|AddUnique (LocalVariables|GetNames%20Local) _array_element_2)))
  (return (Private|GetFileNamesforProject :Append%20Array (Variables|Project|GetSubfolders) :Project%20Names (LocalVariables|GetNames%20Local))))

(fn GetVariationDataAssets
  (Utilities|Array|Clear (LocalVariables|GetData%20Assets%20Local))
  (for _array_element (Variables|Project|GetProjects)
    (bind _returnvalue (Utilities|LoadAssetBlocking _array_element))
    (bind _asbp_animation_browser_project (Utilities|Casting|CastToBP_AnimationBrowser_Project _returnvalue))
    (bind _data_assets (DataAssets|GetVariationsDataAsset _asbp_animation_browser_project))
    (LocalVariables|SetData%20Assets%20Local _data_assets))
  (for _array_element_2 (Variables|Project|GetVariations)
    (bind _returnvalue_2 (Utilities|LoadAssetBlocking _array_element_2))
    (bind _asbp_animation_browser_file_name_base (Utilities|Casting|CastToBP_AnimationBrowser_FileNameBase _returnvalue_2))
    (Utilities|Array|Add (LocalVariables|GetData%20Assets%20Local) _asbp_animation_browser_file_name_base))
  (return (LocalVariables|GetData%20Assets%20Local)))

(fn GetVariations
  (Utilities|Array|Clear (LocalVariables|GetNames%20Local))
  (for _array_element (Variables|Project|GetProjects)
    (bind _returnvalue (Utilities|LoadAssetBlocking _array_element))
    (bind _asbp_animation_browser_project (Utilities|Casting|CastToBP_AnimationBrowser_Project _returnvalue))
    (Utilities|Array|AppendArray (LocalVariables|GetNames%20Local) (FileNaming|GetVariations _asbp_animation_browser_project :!class /AnimInEdTools/AnimationBrowser/ProjectSettings/BP_AnimationBrowser_Project.BP_AnimationBrowser_Project_C)))
  (return (Private|GetFileNamesforProject :Append%20Array (Variables|Project|GetVariations) :Project%20Names (LocalVariables|GetNames%20Local))))
