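; AnimatableCreature_ShotGrid is a small function graph for the animatable creature in the animation editor tools.
; Its GetSGEnabled function returns the Shotgrid Enabled variable, exposing whether ShotGrid integration is switched on.
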(fn GetSGEnabled
  (return (Variables|Default|GetShotgrid%20Enabled)))
