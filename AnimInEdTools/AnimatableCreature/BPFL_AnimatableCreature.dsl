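; Blueprint function library for the AnimInEdTools editor animation tools that queries AnimatableCreature assets and builds animation names.
; Getters block-load creature, prop, ShotGrid and tag data through the Asset Registry, keeping enabled assets and filtering by project when given.
; Finders match on asset and variant or on AnimSequence path and sequencer actor; "All" best-fit prefers Shared or Core, though flagged as unreliable.
; String helpers assemble underscore-joined file names, path-part maps and camel-cased segments.
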(fn AnimatableCreatureFind
  (params (AnimatableCreatures :type Array%20of%20Animatable%20Creature%20Object%20References) (Asset :type String) (Variant :type String))
  (if (Utilities|String|EqualExactly(String) Variant "All")
    (LocalVariables|SetBest%20Fit 0)
    ;[1 1 1] If All is passed in as the varient, then find the "best fit"
    (bind _array_element (Utilities|Array|ForEachLoopwithBreak AnimatableCreatures)
      (:LoopBody
        ;[1 0 0.08962 0.5] Info: When "Variant" is set to all, this best-fit doesn't seem to choose correctly.
        (if (Utilities|String|EqualExactly(String) (Variables|/AnimInEdTools/AnimatableCreature/AnimatableCreature.AnimatableCreature_C|GetAsset _array_element) (Variables|GetAsset))
          (if (or (Utilities|String|Equal,CaseInsensitive(String) (Variables|/AnimInEdTools/AnimatableCreature/AnimatableCreature.AnimatableCreature_C|GetVariant _array_element) "Shared") (Utilities|String|Equal,CaseInsensitive(String) (Variables|/AnimInEdTools/AnimatableCreature/AnimatableCreature.AnimatableCreature_C|GetVariant _array_element) "Core"))
            (return _array_element false)
            (else
              (LocalVariables|SetBest%20Fit _array_element)))))
      (:Completed
        ;;[1 1 1] Return Last option
        (return (LocalVariables|GetBest%20Fit) false)))
    (else
      (for _array_element_2 AnimatableCreatures
        ;[1 1 1] When Variation is left blank
        (if (and (Utilities|String|Equal,CaseInsensitive(String) (Variables|GetAsset) (Variables|/AnimInEdTools/AnimatableCreature/AnimatableCreature.AnimatableCreature_C|GetAsset _array_element_2)) (Utilities|String|Equal,CaseInsensitive(String) (Variables|/AnimInEdTools/AnimatableCreature/AnimatableCreature.AnimatableCreature_C|GetVariant _array_element_2) (Variables|GetVariant)))
          (return _array_element_2 true)))
      (return 0 false))))

(fn CamelCaseString
  (params (String :type String))
  (bind _local_string (LocalVariables|GetLocal%20String))
  (LocalVariables|SetLocal%20String (Private|UpperFirstCharacter String))
  (for _array_element (Utilities|Array|MakeArray "_" " " ".")
    (while (Utilities|String|Contains (LocalVariables|GetLocal%20String) _array_element)
      (LocalVariables|SetLeft%20Split (Utilities|String|Split _local_string _array_element))
      (LocalVariables|SetRight%20Split (.RightS (Utilities|String|Split _local_string _array_element)))
      (if (<= (Utilities|Array|Length (Utilities|String|GetCharacterArrayfromString (.RightS (Utilities|String|Split _local_string _array_element)))) 1)
        (LocalVariables|SetLocal%20String (Utilities|String|Append (LocalVariables|GetLeft%20Split) (Utilities|String|ToUpper (LocalVariables|GetRight%20Split))))
        (else
          (LocalVariables|SetLocal%20String (Utilities|String|Append (LocalVariables|GetLeft%20Split) (Private|UpperFirstCharacter (.RightS (Utilities|String|Split _local_string _array_element)))))))))
  (return (LocalVariables|GetLocal%20String)))

(fn FindAnimatableCreature
  (params (Asset :type String) (Variation :type String) (Project :type BP%20Animation%20Browser%20Project%20Object%20Reference))
  (LocalVariables|SetFound%20Animatable%20Creature 0)
  (bind _array_element (Utilities|Array|ForEachLoopwithBreak (CallFunction|GetAnimatableCreatures Project))
    (:LoopBody
      (if (and (Utilities|String|EqualExactly(String) (Variables|/AnimInEdTools/AnimatableCreature/AnimatableCreature.AnimatableCreature_C|GetAsset _array_element) (Variables|GetAsset)) (Utilities|String|EqualExactly(String) (Variables|/AnimInEdTools/AnimatableCreature/AnimatableCreature.AnimatableCreature_C|GetVariant _array_element) (Variables|GetVariation)))
        (LocalVariables|SetFound%20Animatable%20Creature _array_element)))
    (:Completed
      _))
  (return (LocalVariables|GetFound%20Animatable%20Creature)))

(fn FindAnimatableCreaturefromAnimSequence
  (params (AnimSequence%20Path :type String) (Blueprint%20Actor :type Object%20Class%20Reference) (Project :type BP%20Animation%20Browser%20Project%20Object%20Reference))
  (LocalVariables|SetFound%20Animatable%20Creature 0)
  (bind _array_element (Utilities|Array|ForEachLoopwithBreak (CallFunction|GetAnimatableCreatures Project))
    (:LoopBody
      (if (Utilities|String|StartsWith (Variables|GetAnimSequence%20Path) (Variables|/AnimInEdTools/AnimatableCreature/AnimatableCreature.AnimatableCreature_C|GetAnimation%20Path _array_element))
        (if (== (Variables|/AnimInEdTools/AnimatableCreature/AnimatableCreature.AnimatableCreature_C|GetSequencer%20Actor _array_element) (Utilities|ResolveSoftReference (Variables|GetBlueprint%20Actor)))
          (LocalVariables|SetFound%20Animatable%20Creature _array_element))))
    (:Completed
      _))
  (return (LocalVariables|GetFound%20Animatable%20Creature)))

(fn FindMatchingAnimatableCreatures
  (params (Asset :type String) (Variation :type String) (Project :type BP%20Animation%20Browser%20Project%20Object%20Reference) (bSearchSubClasses :type Boolean))
  (Utilities|Array|Clear (LocalVariables|GetFound%20Animatable%20Creature))
  (bind _array_element (Utilities|Array|ForEachLoopwithBreak (CallFunction|GetAnimatableCreatures Project bSearchSubClasses))
    (:LoopBody
      (if (and (Utilities|String|EqualExactly(String) (Variables|/AnimInEdTools/AnimatableCreature/AnimatableCreature.AnimatableCreature_C|GetAsset _array_element) (Variables|GetAsset)) (or (Utilities|String|EqualExactly(String) (Variables|/AnimInEdTools/AnimatableCreature/AnimatableCreature.AnimatableCreature_C|GetVariant _array_element) (Variables|GetVariation)) (Utilities|String|EqualExactly(String) (Variables|GetVariation) "All")))
        (Utilities|Array|AddUnique (LocalVariables|GetFound%20Animatable%20Creature) _array_element)))
    (:Completed
      _))
  (return (LocalVariables|GetFound%20Animatable%20Creature)))

(fn GenerateFileName
  (params (Asset :type String) (Anim%20Set :type String) (Name :type String) (Descriptor :type String) (Direction :type String) (Variation :type String) (Remove%20Underscores%20in%20Name :type Boolean))
  (return (Utilities|String|Append (Variables|GetAsset) (select (Utilities|String|IsEmpty (Variables|GetAnim%20Set)) "" (Utilities|String|Append "_" (Variables|GetAnim%20Set))) (select (Utilities|String|IsEmpty (select (Variables|GetRemove%20Underscores%20in%20Name) (Private|CamelCaseString (Variables|GetName)) (Private|RemoveSpacesString (Variables|GetName)))) "" (Utilities|String|Append "_" (select (Variables|GetRemove%20Underscores%20in%20Name) (Private|CamelCaseString (Variables|GetName)) (Private|RemoveSpacesString (Variables|GetName))))) (select (Utilities|String|IsEmpty (Variables|GetDescriptor)) "" (Utilities|String|Append "_" (Private|CamelCaseString (Variables|GetDescriptor)))) (select (Utilities|String|IsEmpty (Variables|GetDirection)) "" (Utilities|String|Append "_" (Variables|GetDirection))) (select (Utilities|String|IsEmpty (Variables|GetVariation)) "" (Utilities|String|Append "_" (Variables|GetVariation))))))

(fn GetAnimatableCreatureTags
  (params (Project :type BP%20Animation%20Browser%20Project%20Object%20Reference) (bSearchSubClasses :type Boolean :default true))
  (LocalVariables|SetTag%20Container)
  (bind _outassetdata (AssetRegistry|GetAssetsbyClass (AssetRegistry|GetAssetRegistry) (Utilities|GetSoftClassTopLevelAssetPath "/AnimInEdTools/AnimatableCreature/AnimatableCreature.AnimatableCreature_C") bSearchSubClasses))
  (for _array_element _outassetdata
    (bind (_array_element_packagename _array_element_packagepath _array_element_assetname) (Utilities|Struct|BreakAssetData _array_element))
    (bind _returnvalue (Utilities|LoadAssetBlocking (Utilities|ToSoftObjectReference (Utilities|MakeSoftClassPath (Utilities|String|Append (Utilities|String|ToString(Name) _array_element_packagename) "." (Utilities|String|ToString(Name) _array_element_assetname))))))
    (bind _asanimatable_creature (Utilities|Casting|CastToAnimatableCreature _returnvalue))
    (if (Variables|/AnimInEdTools/AnimatableCreature/AnimatableCreature.AnimatableCreature_C|GetEnabled _asanimatable_creature)
      (if (Utilities|IsValid (Variables|GetProject) :!class /Script/Engine.KismetSystemLibrary)
        (if (Utilities|Array|ContainsItem (Class|AnimatableCreature|GetProjects _asanimatable_creature :!class /AnimInEdTools/AnimatableCreature/AnimatableCreature.AnimatableCreature_C) (Variables|GetProject))
          (GameplayTags|AppendGameplayTagContainers (LocalVariables|GetTag%20Container) (Variables|/AnimInEdTools/AnimatableCreature/AnimatableCreature.AnimatableCreature_C|GetTags _asanimatable_creature)))
        (else
          ;[1 1 1] If no project is given, return all tags
          (GameplayTags|AppendGameplayTagContainers (LocalVariables|GetTag%20Container) (Variables|/AnimInEdTools/AnimatableCreature/AnimatableCreature.AnimatableCreature_C|GetTags _asanimatable_creature))))))
  (return (LocalVariables|GetTag%20Container)))

(fn GetAnimatableCreatures
  (params (Project :type BP%20Animation%20Browser%20Project%20Object%20Reference) (bSearchSubClasses :type Boolean :default true))
  (Utilities|Array|Clear (LocalVariables|GetLocal%20Animatable%20Creatures))
  (bind _outassetdata (AssetRegistry|GetAssetsbyClass (AssetRegistry|GetAssetRegistry) (Utilities|GetSoftClassTopLevelAssetPath "/AnimInEdTools/AnimatableCreature/AnimatableCreature.AnimatableCreature_C") bSearchSubClasses))
  (for _array_element _outassetdata
    (bind (_array_element_packagename _array_element_packagepath _array_element_assetname) (Utilities|Struct|BreakAssetData _array_element))
    (bind _returnvalue (Utilities|LoadAssetBlocking (Utilities|ToSoftObjectReference (Utilities|MakeSoftClassPath (Utilities|String|Append (Utilities|String|ToString(Name) _array_element_packagename) "." (Utilities|String|ToString(Name) _array_element_assetname))))))
    (bind _asanimatable_creature (Utilities|Casting|CastToAnimatableCreature _returnvalue))
    (if (Variables|/AnimInEdTools/AnimatableCreature/AnimatableCreature.AnimatableCreature_C|GetEnabled _asanimatable_creature)
      (if (Utilities|IsValid (Variables|GetProject) :!class /Script/Engine.KismetSystemLibrary)
        (if (Utilities|Array|ContainsItem (Class|AnimatableCreature|GetProjects _asanimatable_creature :!class /AnimInEdTools/AnimatableCreature/AnimatableCreature.AnimatableCreature_C) (Variables|GetProject))
          (Utilities|Array|Add (LocalVariables|GetLocal%20Animatable%20Creatures) _asanimatable_creature))
        (else
          ;[1 1 1] If no project is given, return all creatures
          (Utilities|Array|Add (LocalVariables|GetLocal%20Animatable%20Creatures) _asanimatable_creature)))))
  (return (LocalVariables|GetLocal%20Animatable%20Creatures)))

(fn GetAnimatableProps
  (params (Project :type BP%20Animation%20Browser%20Project%20Object%20Reference))
  (Utilities|Array|Clear (LocalVariables|GetLocal%20Animatable%20Creatures))
  (bind _outassetdata (AssetRegistry|GetAssetsbyClass (AssetRegistry|GetAssetRegistry) (Utilities|GetSoftClassTopLevelAssetPath "/AnimInEdTools/AnimatableCreature/AnimatableProp.AnimatableProp_C") false))
  (for _array_element _outassetdata
    (bind (_array_element_packagename _array_element_packagepath _array_element_assetname) (Utilities|Struct|BreakAssetData _array_element))
    (bind _returnvalue (Utilities|LoadAssetBlocking (Utilities|ToSoftObjectReference (Utilities|MakeSoftClassPath (Utilities|String|Append (Utilities|String|ToString(Name) _array_element_packagename) "." (Utilities|String|ToString(Name) _array_element_assetname))))))
    (bind _asanimatable_prop (Utilities|Casting|CastToAnimatableProp _returnvalue))
    (if (Variables|/AnimInEdTools/AnimatableCreature/AnimatableCreature.AnimatableCreature_C|GetEnabled _asanimatable_prop)
      (if (Utilities|IsValid (Variables|GetProject) :!class /Script/Engine.KismetSystemLibrary)
        (if (Utilities|Array|ContainsItem (Class|AnimatableCreature|GetProjects _asanimatable_prop :!class /AnimInEdTools/AnimatableCreature/AnimatableCreature.AnimatableCreature_C) (Variables|GetProject))
          (Utilities|Array|Add (LocalVariables|GetLocal%20Animatable%20Creatures) _asanimatable_prop))
        (else
          ;[1 1 1] If no project is given, return all creatures
          (Utilities|Array|Add (LocalVariables|GetLocal%20Animatable%20Creatures) _asanimatable_prop)))))
  (return (LocalVariables|GetLocal%20Animatable%20Creatures)))

(fn GetAnimationDirectoryPath
  (params (Asset :type String) (Variation :type String%20%28by%20ref%29) (Project :type BP%20Animation%20Browser%20Project%20Object%20Reference))
  (bind _animatable_creatures (CallFunction|FindAnimatableCreature Asset Variation Project))
  (Utilities|IsValid _animatable_creatures
    (:Is%20Valid
      (return (Variables|/AnimInEdTools/AnimatableCreature/AnimatableCreature.AnimatableCreature_C|GetAnimation%20Path _animatable_creatures)))
    (:Is%20Not%20Valid
      (return ""))))

(fn GetPathPartsMap
  (params (Asset :type String) (Animation%20Set :type String) (Animation%20Name :type String) (Descriptor :type String) (Direction :type String) (Subfolder :type String) (Variation :type String) (Plugin :type String))
  (return (Utilities|Map|MakeMap "Asset" Asset "AnimationSet" Animation%20Set "AnimationName" Animation%20Name "Descriptor" Descriptor "Direction" Direction "Subfolder" Subfolder "Variation" Variation "Plugin" Plugin)))

(fn GetSGAnimatableCreatures
  (params (Project :type BP%20Animation%20Browser%20Project%20Object%20Reference))
  (bind _outassetdata (AssetRegistry|GetAssetsbyClass (AssetRegistry|GetAssetRegistry) (Utilities|GetSoftClassTopLevelAssetPath "/AnimInEdTools/AnimatableCreature/AnimatableCreature_ShotGrid.AnimatableCreature_ShotGrid_C") false))
  (for _array_element _outassetdata
    (bind (_array_element_packagename _array_element_packagepath _array_element_assetname) (Utilities|Struct|BreakAssetData _array_element))
    (bind _returnvalue (Utilities|LoadAssetBlocking (Utilities|ToSoftObjectReference (Utilities|MakeSoftClassPath (Utilities|String|Append (Utilities|String|ToString(Name) _array_element_packagename) "." (Utilities|String|ToString(Name) _array_element_assetname))))))
    (bind _asanimatable_creature_shot_grid (Utilities|Casting|CastToAnimatableCreature_ShotGrid _returnvalue))
    (if (and (Variables|/AnimInEdTools/AnimatableCreature/AnimatableCreature.AnimatableCreature_C|GetEnabled _asanimatable_creature_shot_grid) (Class|AnimatableCreatureShotGrid|GetSGEnabled _asanimatable_creature_shot_grid))
      (if (Utilities|IsValid (Variables|GetProject) :!class /Script/Engine.KismetSystemLibrary)
        (if (Utilities|Array|ContainsItem (Class|AnimatableCreature|GetProjects _asanimatable_creature_shot_grid :!class /AnimInEdTools/AnimatableCreature/AnimatableCreature.AnimatableCreature_C) (Variables|GetProject))
          (Utilities|Array|Add (LocalVariables|GetLocal%20Animatable%20Creatures) _asanimatable_creature_shot_grid))
        (else
          ;[1 1 1] If no project is given, return all creatures
          (Utilities|Array|Add (LocalVariables|GetLocal%20Animatable%20Creatures) _asanimatable_creature_shot_grid)))))
  (return (LocalVariables|GetLocal%20Animatable%20Creatures)))

(fn GetVariationsforCreature
  (params (Creature :type String) (Project :type BP%20Animation%20Browser%20Project%20Object%20Reference))
  (for _array_element (CallFunction|GetAnimatableCreatures Project)
    (if (Utilities|String|EqualExactly(String) (Variables|/AnimInEdTools/AnimatableCreature/AnimatableCreature.AnimatableCreature_C|GetAsset _array_element) (Variables|GetCreature))
      (Utilities|Array|Add (LocalVariables|GetMatching%20Animatable%20Creatures) _array_element)))
  (return (LocalVariables|GetMatching%20Animatable%20Creatures)))

(fn RemoveSpacesString
  (params (String :type String))
  (return (Utilities|String|Replace String " " "_")))

(fn UpperFirstCharacter
  (params (SourceString :type String))
  (bind _sourcestring (Variables|GetSourceString))
  (bind _returnvalue (Utilities|String|GetCharacterArrayfromString _sourcestring))
  (if (!= (Utilities|String|Len SourceString) 0)
    (bind _output_get (LocalVariables|SetString%20Array _returnvalue))
    (Utilities|Array|SetArrayElem _output_get 0 (Utilities|String|ToUpper (Utilities|Array|Get(aref) _returnvalue)))
    (return (Utilities|String|JoinStringArray _output_get ""))
    (else
      (return (Variables|GetSourceString)))))
