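# Editor utility function library that creates Level Sequences and Anim Sequences for Animatable Creature assets in the Animation Browser.
# Creation functions add the creature as a spawnable with Control Rig track, export a linked Anim Sequence, copy project compression settings and save browser selection.
# The level-sequence flow also opens the default map and pastes default poses onto unkeyed rigs.
# Embedded Python handles asset creation, export and token-based path resolution.
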EditorUtilityBlueprint /AnimInEdTools/AnimatableCreature/EUFL_AnimatableCreature
graphs [
 Add Actor {nodes [
   n1 = Add Actor!() -> (then, Animatable Creature:Animatable Creature Object Reference, Level Sequence:Level Sequence Object Reference, bIsLayeredControlRig:bool=~) {is=title K2Node_FunctionEntry}
   > n2 = LoadClassAsset_Blocking(AssetClass=Get Sequencer Actor!(self:Animatable Creature Object Reference=@n1) -> (Sequencer Actor:Object Soft Class Reference))
   > n3 = Load Asset Blocking!(
    execute
    InRef:"Control Rig Asset Soft Reference Structure (by ref)"=Get Control Rig!(self:Animatable Creature Object Reference=@n1) -> (Control Rig:Control Rig Asset Soft Reference Structure)
    bForceLoad:bool=true
   ) -> (then, ReturnValue:Control Rig Asset Strong Reference Structure)
   > n4 = Add Actor Base!(
    execute
    Level Sequence:Level Sequence Object Reference=Get Level Sequence() -> (Level Sequence:Level Sequence Object Reference)
    bIsLayeredControlRig:bool=Get bIsLayeredControlRig() -> (bIsLayeredControlRig:bool)
    ControlRigClass:Control Rig Asset Strong Reference Structure=@n3
    Actor Class:Object Class Reference=@n2
    __WorldContext:Object Reference
   ) -> (then, New Binding:Movie Scene Binding Proxy Structure, Track:Movie Scene Track Object Reference)
   > Return Node!(execute, New Binding:Movie Scene Binding Proxy Structure=@n4, Track:Movie Scene Track Object Reference=@n4.Track)
  ]}
 Add Actor Base {nodes [
   n1 = Add Actor Base!() -> (
    then
    Level Sequence:Level Sequence Object Reference
    bIsLayeredControlRig:bool=~
    ControlRigClass:Control Rig Asset Strong Reference Structure
    Actor Class:Object Class Reference
   ) {is=title K2Node_FunctionEntry}
   > n2 = AddSpawnableFromClass!(
    execute
    self:Level Sequence Editor Subsystem Object Reference=Get LevelSequenceEditorSubsystem() -> (ReturnValue:Level Sequence Editor Subsystem Object Reference) {is=title K2Node_GetEditorSubsystem}
    Sequence:Movie Scene Sequence Object Reference=@n1
    ClassToSpawn:Object Class Reference=@n1.Actor Class
   ) -> (then, ReturnValue:Movie Scene Binding Proxy Structure)
   > n3 = Branch(Condition=Is Valid Referfence!(InRef:"Control Rig Asset Strong Reference Structure (by ref)"=@n1.ControlRigClass) -> (ReturnValue:bool))
   > n4 = GetEditorWorld(self=Get UnrealEditorSubsystem())
   > n5 = FindOrCreateControlRigTrackFromReference!(
    execute
    World:World Object Reference=@n4
    LevelSequence:Level Sequence Object Reference=Get Level Sequence()
    InRef:"Control Rig Asset Strong Reference Structure (by ref)"=@n1.ControlRigClass
    InBinding:"Movie Scene Binding Proxy Structure (by ref)"=@n2
    bIsLayeredControlRig:bool=Get bIsLayeredControlRig()
   ) -> (then, ReturnValue:Movie Scene Track Object Reference)
   > Return Node(New Binding=@n2, Track=@n5)
   Return Node(execute=@n3.else, New Binding=@n2)
  ]}
 Create Anim Sequence From Path {nodes [
   n1 = Create Anim Sequence From Path!() -> (then, Anim Sequence Path:String, skeleton:Skeleton Object Reference) {is=title K2Node_FunctionEntry}
   > n2 = Execute Python Script!(
    execute
    PythonScript="import unreal\r\n\r\n## Grab the Level Editor World\r\neditor_subsystem = unreal.get_editor_subsystem(unreal.UnrealEditorSubsystem)\r\nworld = editor_subsystem.get_editor_world()\r\n \r\n## Create animation sequence export options\r\nanim_seq_export_options = unreal.AnimSeqExportOption()\r\nanim_seq_export_options.export_transforms = True\r\nanim_seq_export_options.export_morph_targets = True\r\n \r\n## Get asset tools\r\nasset_tools = unreal.AssetToolsHelpers.get_asset_tools()\r\n\r\n## Create an anim sequence factory, assign it the target skeleton\r\nanim_sequence_factory = unreal.AnimSequenceFactory()\r\nanim_sequence_factory.target_skeleton = skeleton\r\nanim_sequence = unreal.AssetTools.create_asset(asset_tools, asset_name=asset_name, package_path=package_path, asset_class=unreal.AnimSequence, factory=anim_sequence_factory)  "
    package_path:String=GetPath!(InPath:String=@n1) -> (ReturnValue:String)
    asset_name:String=GetBaseFilename(InPath=@n1, bRemovePath=true)
    skeleton:Skeleton Object Reference=@n1.skeleton
   ) -> (then, ReturnValue, anim_sequence:Anim Sequence Object Reference) {inputs [package_path,asset_name,skeleton], outputs [anim_sequence]}
   > Return Node!(execute, AnimSeq:Anim Sequence Object Reference=@n2.anim_sequence)
  ]}
 Create Level Sequence from Path {nodes [
   n1 = Create Level Sequence from Path!() -> (then, Level Sequence Path:String) {is=title K2Node_FunctionEntry}
   > n2 = Execute Python Script!(
    execute
    PythonScript="import unreal\r\n\r\n## Get asset tools\r\nasset_tools = unreal.AssetToolsHelpers.get_asset_tools()\r\n \r\n## Creates a Level Sequence\r\nlevel_sequence = unreal.AssetTools.create_asset(asset_tools, asset_name = asset_name, package_path = package_path, asset_class = unreal.LevelSequence, factory = unreal.LevelSequenceFactoryNew())\r\n\r\n## Then open it in sequencer\r\nunreal.LevelSequenceEditorBlueprintLibrary.open_level_sequence(level_sequence)"
    asset_name=GetCleanFilename!(InPath:String=@n1) -> (ReturnValue:String)
    package_path=GetPath(InPath=@n1)
   ) -> (then, ReturnValue, level_sequence:Object Reference) {inputs [asset_name,package_path], outputs [level_sequence]}
   > Return Node!(execute, level_sequence:Object Reference=@n2.level_sequence, ReturnValue:bool=@n2)
  ]}
 Create Linked Anim Sequence {nodes [
   n1 = Create Linked Anim Sequence!() -> (
    then
    Level Sequence:Level Sequence Object Reference
    New AnimSequence Path:String
    Binding:Movie Scene Binding Proxy Structure
    Project:BP Animation Browser Project Object Reference
   ) {is=title K2Node_FunctionEntry}
   > n2 = Get Current Time!(execute, TimeUnit:EMovieSceneTimeUnit Enum=DisplayRate) -> (then, ReturnValue:Movie Scene Sequence Playback Params Structure) {is=title K2Node_CallFunction}
   n3 = Break Movie Scene Sequence Playback Params!(MovieSceneSequencePlaybackParams:"Movie Scene Sequence Playback Params Structure (by ref)"=@n2) -> (
    Frame_FrameNumber_Value:int
    Frame_SubFrame:float=~
    Time:float=~
    MarkedFrame:String
    Timecode:Timecode Structure
    PositionType:EMovieScenePositionType Enum
    UpdateMethod:EUpdatePositionMethod Enum
    bHasJumped:bool=~
   ) {is=title K2Node_BreakStruct}
   > Set Current Time!(
    execute
    PlaybackParams:Movie Scene Sequence Playback Params Structure=Make Movie Scene Sequence Playback Params!(
     Frame_FrameNumber_Value:int=int + int(A=@n3)
     Frame_SubFrame:float=0
     Time:float=0
     MarkedFrame:String
     Timecode:Timecode Structure=(Hours=0,Minutes=0,Seconds=0,Frames=0,Subframe=0,bDropFrameFormat=False)
     PositionType:EMovieScenePositionType Enum=Frame
     UpdateMethod:EUpdatePositionMethod Enum=Play
     bHasJumped:bool=False
    ) -> (MovieSceneSequencePlaybackParams:Movie Scene Sequence Playback Params Structure) {is=title K2Node_MakeStruct}
    TimeUnit:EMovieSceneTimeUnit Enum=DisplayRate
   ) -> (then) {is=title K2Node_CallFunction}
   > Set Current Time(PlaybackParams=@n2, TimeUnit=DisplayRate)
   > n4 = GetBindingID!(execute, Sequence:Movie Scene Sequence Object Reference=@n1, InBinding:"Movie Scene Binding Proxy Structure (by ref)"=@n1.Binding) -> (then, ReturnValue:Movie Scene Object Binding ID Structure)
   > n5 = For Each Loop with Break!(
    Exec:Exec
    Array:Array of Object References=GetBoundObjects!(ObjectBinding:Movie Scene Object Binding ID Structure=@n4) -> (ReturnValue:Array of Object References)
    Break:Exec=@n15.then
   ) -> (LoopBody:Exec, Array Element:Object Reference, Array Index:int, Completed:Exec)
   n6 = Is Valid(exec=@n5.Completed, InputObject=Get Skeleton() -> (Skeleton:Skeleton Object Reference))
   Return Node!(execute=@n6.Is Not Valid, Anim Sequence:Anim Sequence Object Reference)
   > Execute Python Script!(
    execute
    PythonScript="import unreal\r\n\r\n## Grab the Level Editor World\r\neditor_subsystem = unreal.get_editor_subsystem(unreal.UnrealEditorSubsystem)\r\nworld = editor_subsystem.get_editor_world()\r\n \r\n## Create animation sequence export options\r\nanim_seq_export_options = unreal.AnimSeqExportOption()\r\nanim_seq_export_options.export_transforms = True\r\nanim_seq_export_options.export_morph_targets = True\r\n \r\n## Get asset tools\r\nasset_tools = unreal.AssetToolsHelpers.get_asset_tools()\r\n\r\n## Create an anim sequence factory, assign it the target skeleton\r\nanim_sequence_factory = unreal.AnimSequenceFactory()\r\nanim_sequence_factory.target_skeleton = skeleton\r\nanim_sequence = unreal.AssetTools.create_asset(asset_tools, asset_name=asset_name, package_path=package_path, asset_class=unreal.AnimSequence, factory=anim_sequence_factory)  \r\n\r\n## Export\r\nunreal.ControlRigSequencerLibrary.export_anim_sequence_from_sequencer(anim_sequence, anim_seq_export_options, binding, True)"
    package_path=GetPath(InPath=@n1.New AnimSequence Path)
    asset_name=GetBaseFilename(InPath=@n1.New AnimSequence Path, bRemovePath=true)
    level_sequence:Level Sequence Object Reference=@n1
    binding:Movie Scene Binding Proxy Structure=@n1.Binding
    skeleton:Skeleton Object Reference=Get Skeleton()
   ) -> (then, ReturnValue) {inputs [package_path, asset_name, level_sequence, binding, skeleton]}
   > n7 = LoadAsset(AssetPath=Get New AnimSequence Path() -> (New AnimSequence Path:String))
   > n8 = Cast To AnimSequence!(execute, Object=@n7) -> (then, CastFailed, AsAnim Sequence:Anim Sequence Object Reference)
   > n9 = Sequence()
   n10 = Get Compression Settings!(self:BP Animation Browser Project Object Reference=@n1.Project) -> (Compression Settings:Anim Sequence Soft Object Reference)
   > Branch(Condition=IsValidSoftObjectReference(SoftObjectReference=@n10))
   > n11 = LoadAsset_Blocking(Asset=@n10)
   > n12 = Cast To AnimSequence(Object=@n11)
   > SetBoneCompressionSettings!(
    execute
    AnimationSequence:Anim Sequence Object Reference=@n8
    CompressionSettings:Anim Bone Compression Settings Object Reference=GetBoneCompressionSettings!(AnimationSequence:Anim Sequence Object Reference=@n12) -> (CompressionSettings:Anim Bone Compression Settings Object Reference)
   ) -> (then)
   > SetCurveCompressionSettings!(
    execute
    AnimationSequence:Anim Sequence Object Reference=@n8
    CompressionSettings:Anim Curve Compression Settings Object Reference=GetCurveCompressionSettings!(AnimationSequence:Anim Sequence Object Reference=@n12) -> (CompressionSettings:Anim Curve Compression Settings Object Reference)
   ) -> (then)
   > SetAnimationInterpolationType!(
    execute
    AnimationSequence:Anim Sequence Object Reference=@n8
    InterpolationType:EAnimInterpolationType Enum=GetAnimationInterpolationType!(AnimationSequence:Anim Sequence Object Reference=@n12) -> (InterpolationType:EAnimInterpolationType Enum=Linear)
   ) -> (then)
   Return Node(execute=@n9.then_1, Anim Sequence=@n8)
   n13 = Cast To Actor!(execute=@n5.LoopBody, Object=@n5) -> (then, CastFailed, AsActor:Actor Object Reference) {is=title K2Node_DynamicCast}
   n14 = GetComponentByClass!(self:Actor Object Reference=@n13, ComponentClass:Actor Component Class Reference=/Script/Engine.SkeletalMeshComponent) -> (ReturnValue:Skeletal Mesh Component Object Reference)
   > n15 = Set Skeleton(Skeleton:Skeleton Object Reference=Get Skeleton!(self:Skeletal Mesh Object Reference=Get SkeletalMesh!(self:Skinned Mesh Component Object Reference=@n14) -> (SkeletalMesh:Skeletal Mesh Object Reference)) -> (Skeleton)) -> (Output_Get:Skeleton Object Reference)
  ]}
 Create New Anim Sequence {nodes [
   n1 = Create New Anim Sequence!() -> (
    then
    AC:Animatable Creature Object Reference
    Anim Sequence Path:String
    Level Sequence Path:String
    Project:BP Animation Browser Project Object Reference
    Save Game:SG Animation Browser Object Reference
   ) {is=title K2Node_FunctionEntry}
   > Set Anim Seq Path(Anim Seq Path:String=@n1.Anim Sequence Path) -> (Output_Get:String)
   > n2 = Set Animatable Creature(Animatable Creature:Animatable Creature Object Reference=@n1) -> (Output_Get:Animatable Creature Object Reference)
   > n3 = Is Valid(InputObject=@n2)
   PrintString(execute=@n3.Is Not Valid, InString="Error: Invalid AC file for `Create New Animation()`", bPrintToScreen=false)
   n4 = LoadAsset(execute=@n3.Is Valid, AssetPath=Get Anim Seq Path() -> (Anim Seq Path:String))
   > n5 = Cast To AnimSequence(Object=@n4)
   > n6 = Set Anim Seq(Anim Seq:Anim Sequence Object Reference=@n5) -> (Output_Get:Anim Sequence Object Reference)
   n7 = LoadAsset(execute=[@n6.then, @n5.CastFailed], AssetPath=@n1.Level Sequence Path)
   > n8 = Cast To LevelSequence!(execute, Object=@n7) -> (then, CastFailed, AsLevel Sequence:Level Sequence Object Reference) {is=title K2Node_DynamicCast}
   > n9 = Set Level Seq(Level Seq:Level Sequence Object Reference=@n8) -> (Output_Get:Level Sequence Object Reference)
   n10 = Is Valid(exec=[@n9.then, @n8.CastFailed], InputObject=@n9)
   n11 = Get SKM From AC File!(
    execute=@n10.Is Not Valid
    AC file:Animatable Creature Object Reference=Get Animatable Creature() -> (Animatable Creature:Animatable Creature Object Reference)
    __WorldContext:Object Reference
   ) -> (then, Skeletal Mesh:Skeletal Mesh Object Reference) {is=title K2Node_CallFunction}
   n12 = Get Skeleton(self=@n11)
   > Set sk(sk:Skeleton Object Reference=@n12) -> (Output_Get:Skeleton Object Reference)
   n13 = Get Anim Seq() -> (Anim Seq:Anim Sequence Object Reference)
   > n14 = Branch(Condition=AND Boolean(
     A=IsValid(Object=@n13)
     B="Not Equal (Object)"(A:Object Reference=@n13, B:Object Reference) -> (ReturnValue:bool=~) {is=title K2Node_PromotableOperator}
    ))
   Return Node!(
    execute=[@n19.then_1, @n17.then, @n14.then, @n16.Is Not Valid]
    LevelSequence:Level Sequence Object Reference=Get Level Seq() -> (Level Seq:Level Sequence Object Reference)
    AnimSequence:Anim Sequence Object Reference=Get Anim Seq()
   )
   n15 = Create Anim Sequence From Path!(execute=@n14.else, Anim Sequence Path:String=Get Anim Seq Path(), skeleton:Skeleton Object Reference=@n12, __WorldContext:Object Reference) -> (then, AnimSeq:Anim Sequence Object Reference) {is=title K2Node_CallFunction}
   > n16 = Is Valid(InputObject=@n15)
   > n17 = Set Anim Seq(Anim Seq=@n15)
   n18 = Is Valid(exec=@n10.Is Valid, InputObject=Get Anim Seq())
   n19 = Sequence(execute=@n18.Is Not Valid)
   > n20 = Add Actor!(
    execute
    Animatable Creature:Animatable Creature Object Reference=Get Animatable Creature()
    Level Sequence:Level Sequence Object Reference=Get Level Seq()
    bIsLayeredControlRig:bool=false
    __WorldContext:Object Reference
   ) -> (then, New Binding:Movie Scene Binding Proxy Structure, Track:Movie Scene Track Object Reference) {is=title K2Node_CallFunction}
   > n21 = Create Linked Anim Sequence!(
    execute
    Level Sequence:Level Sequence Object Reference=Get Level Seq()
    New AnimSequence Path:String=Get Anim Sequence Path() -> (Anim Sequence Path:String)
    Binding:Movie Scene Binding Proxy Structure=@n20
    Project:BP Animation Browser Project Object Reference=@n1.Project
    __WorldContext:Object Reference
   ) -> (then, Anim Sequence:Anim Sequence Object Reference) {is=title K2Node_CallFunction}
   > Set Anim Seq(Anim Seq=@n21)
   > SaveLoadedAssets(
    AssetsToSave=Make Array!("[0]":Object Reference=Get Level Seq(), "[1]":Object Reference=@n21) -> (Array:Array of Object References)
    bOnlyIfIsDirty=false
   )
   > Is Valid(InputObject=Get Save Game() -> (Save Game:SG Animation Browser Object Reference))
   n22 = Get Save Game()
   > Set Selected Path!(
    execute
    Selected Path:String=GetSoftObjectPath!(Object=@n21) -> (ReturnValue_PathString:String)
    self:SG Animation Browser Object Reference=@n22
   ) -> (then, Output_Get:String)
   > Set Selected Index!(execute, Selected Index:int=-1, self:SG Animation Browser Object Reference=@n22) -> (then, Output_Get:int=0)
   > Set Apply Default Pose!(execute, Apply Default Pose:bool=true, self:SG Animation Browser Object Reference=@n22) -> (then, Output_Get:bool)
   > SaveGameToSlot!(execute, SaveGameObject:Save Game Object Reference=@n22, SlotName:String=AnimationBrowser, UserIndex:int=0) -> (then, ReturnValue:bool)
   > SaveLoadedAsset(AssetToSave=Get Level Seq(), bOnlyIfIsDirty=true)
   n23 = Get SKM From AC File(execute=@n18.Is Valid, AC file=Get Animatable Creature())
   > Set sk(sk=Get Skeleton(self=@n23))
   > Return Node(LevelSequence=Get Level Seq(), AnimSequence=Get Anim Seq())
  ]}
 Create New Animation With Level Sequence {nodes [
   n1 = Create New Animation With Level Sequence!() -> (
    then
    AC:Animatable Creature Object Reference
    AnimatableCreatures:Array of Animatable Creature Object References
    Anim Sequence Path:String
    Level Sequence Path:String
    Project:BP Animation Browser Project Object Reference
    Save Game:SG Animation Browser Object Reference
   ) {is=title K2Node_FunctionEntry}
   > Set Level Seq Path(Level Seq Path:String=@n1.Level Sequence Path) -> (Output_Get:String)
   > n2 = Set Animatable Creature(Animatable Creature=@n1)
   > Is Valid(InputObject=@n2)
   > Open Default Animation Map!(execute, Project:BP Animation Browser Project Object Reference=@n1.Project, __WorldContext:Object Reference) -> (then, Map:World Object Reference)
   > n3 = DoesAssetExist(AssetPath=Get Level Seq Path() -> (Level Seq Path:String))
   > n4 = Branch(Condition=@n3)
   > n5 = LoadAsset(AssetPath=Get Level Seq Path())
   > n6 = Cast To LevelSequence(Object=@n5)
   > n7 = Set Level Seq(Level Seq=@n6)
   > n8 = OpenLevelSequence!(execute, LevelSequence:Level Sequence Object Reference=@n7) -> (then, ReturnValue:bool)
   n9 = Add Actor(execute=[@n25.Is Valid, @n8.then], Animatable Creature=Get Animatable Creature(), Level Sequence=Get Level Seq(), bIsLayeredControlRig=false)
   > Set New Binding(New Binding:Movie Scene Binding Proxy Structure=@n9) -> (Output_Get:Movie Scene Binding Proxy Structure)
   > n10 = Sequence!(execute) -> (then_0, then_1, then_2:Exec)
   Return Node(execute=@n10.then_2, LevelSequence=Get Level Seq(), AnimSequence=Get Anim Seq())
   n11 = Get Animatable Creature()
   n12 = AnimatableCreature Find!(
    AnimatableCreatures:Array of Animatable Creature Object References=Get AnimatableCreatures() -> (AnimatableCreatures:Array of Animatable Creature Object References)
    Asset:String=Get Asset!(self:Animatable Creature Object Reference=@n11) -> (Asset:String)
    Variant:String=Get Variant!(self:Animatable Creature Object Reference=@n11) -> (Variant:String)
    __WorldContext:Object Reference
   ) -> (AnimatableCreature:Animatable Creature Object Reference, Exact Match:bool)
   > Is Valid(InputObject=@n12)
   n13 = Get Default Pose!(self:Animatable Creature Object Reference=@n12) -> (Default Pose:Control Rig Pose Asset Soft Object Reference)
   > Branch(Condition=IsValidSoftObjectReference(SoftObjectReference=@n13))
   n14 = Make Frame Number!(Value:int=1) -> (FrameNumber:Frame Number Structure) {is=title K2Node_MakeStruct}
   > Set Current Local Time!(
    execute
    PlaybackParams:Movie Scene Sequence Playback Params Structure=Make Movie Scene Sequence Playback Params!(
     Frame:Frame Time Structure=Make Frame Time!(FrameNumber:Frame Number Structure=@n14, SubFrame:float=0) -> (FrameTime:Frame Time Structure) {is=title K2Node_MakeStruct}
     Time=0
     MarkedFrame
     Timecode=(Hours=0,Minutes=0,Seconds=0,Frames=0,Subframe=0,bDropFrameFormat=False)
     PositionType=Frame
     UpdateMethod=Play
     bHasJumped=False
    ) -> (MovieSceneSequencePlaybackParams)
    TimeUnit:EMovieSceneTimeUnit Enum=DisplayRate
   ) -> (then) {is=title K2Node_CallFunction}
   n15 = Make Frame Number(Value=0)
   > Set Current Local Time(
    PlaybackParams=Make Movie Scene Sequence Playback Params(
     Frame=Make Frame Time(FrameNumber=@n15, SubFrame=0)
     Time=0
     Timecode=(Hours=0,Minutes=0,Seconds=0,Frames=0,Subframe=0,bDropFrameFormat=False)
     PositionType=Frame
     UpdateMethod=Play
     bHasJumped=False
    )
    TimeUnit=DisplayRate
   )
   > n16 = GetControlRigs!(execute, LevelSequence:Level Sequence Object Reference=Get Level Seq()) -> (then, ReturnValue:Array of Control Rig Sequencer Binding Proxy Structures)
   > n17 = For Each Loop with Break(Array:Array of Control Rig Sequencer Binding Proxy Structures=@n16, Break=@n19.then) -> (Array Element:Control Rig Sequencer Binding Proxy Structure)
   n18 = Get New Binding() -> (New Binding:Movie Scene Binding Proxy Structure)
   > Branch(Condition="Equal (Guid)"(
     A:"Guid Structure (by ref)"=Break Movie Scene Binding Proxy!(MovieSceneBindingProxy:"Movie Scene Binding Proxy Structure (by ref)"=@n18) -> (BindingID:Guid Structure, Sequence:Movie Scene Sequence Object Reference) {is=title K2Node_BreakStruct}
     B:"Guid Structure (by ref)"=Break Control Rig Sequencer Binding Proxy!(ControlRigSequencerBindingProxy:"Control Rig Sequencer Binding Proxy Structure (by ref)"=@n17) -> (
      Proxy_BindingID:Guid Structure
      Proxy_Sequence:Movie Scene Sequence Object Reference
      ControlRig:Control Rig Object Reference
      Track:Movie Scene Control Rig Parameter Track Object Reference
     ) {is=title K2Node_BreakStruct}
    ) -> (ReturnValue:bool=~) {is=title K2Node_PromotableOperator})
   > n19 = Paste Pose!(
    execute
    Pose Asset:Soft Object Reference=@n13
    LevelSequence:Level Sequence Object Reference=Get Level Seq()
    ControlRig Proxy Binding:"Control Rig Sequencer Binding Proxy Structure (by ref)"=@n17
    __WorldContext:Object Reference
   ) -> (then)
   n20 = Create Linked Anim Sequence(
    execute=@n10.then_1
    Level Sequence=Get Level Seq()
    New AnimSequence Path=Get Anim Sequence Path()
    Binding=Get New Binding()
    Project=@n1.Project
   )
   > Set Anim Seq(Anim Seq=@n20)
   > SaveLoadedAssets(AssetsToSave=Make Array("[0]"=Get Level Seq(), "[1]"=@n20), bOnlyIfIsDirty=false)
   > Is Valid(InputObject=Get Save Game())
   n21 = Get Save Game()
   > Set Selected Path(Selected Path=GetSoftObjectPath(Object=@n20), self=@n21)
   > Set Selected Index(Selected Index=-1, self=@n21) -> (Output_Get=0)
   > Set Apply Default Pose(Apply Default Pose=true, self=@n21)
   > SaveGameToSlot(SaveGameObject=@n21, SlotName=AnimationBrowser, UserIndex=0)
   n22 = Create Level Sequence from Path!(execute=[@n6.CastFailed, @n4.else], Level Sequence Path:String=Get Level Seq Path(), __WorldContext:Object Reference) -> (then, level_sequence:Object Reference, ReturnValue:bool) {is=title K2Node_CallFunction}
   > n23 = Cast To LevelSequence(Object=@n22)
   > n24 = Set Level Seq(Level Seq=@n23)
   > n25 = Is Valid(InputObject=@n24)
  ]}
 FormatAnimationPathsFromAsset {nodes [
   n1 = FormatAnimationPathsFromAsset!() -> (then, Asset Name:String, AC File:Animatable Creature Object Reference, Feature:String) {is=title K2Node_FunctionEntry}
   > Set loc_AnimName(loc_AnimName:String=Append(A=Anim_, B=@n1)) -> (Output_Get:String)
   > n2 = ReverseTokenMap!(execute, Asset Name:String=@n1, AC File:Animatable Creature Object Reference=@n1.AC File, __WorldContext:Object Reference) -> (then, TokenMap:Map of Strings to Strings)
   > n3 = "For Each Loop (Map)"!(execute:Exec, MapPin:Map of Strings to Strings=@n2, BreakPin:Exec) -> (then:Exec, KeyPin:String, ValuePin:String, CompletedPin:Exec)
   > PrintText(InText=Format Text!(Format="LOCTEXT(\"Resolving Token -- {k} : {v}\")", k:String=@n3, v:String=@n3.ValuePin) -> (Result))
   n4 = Branch(execute=@n3.CompletedPin, Condition=IsEmpty(InString=@n1.Feature))
   n5 = ResolvePathsFromTokens!(
    execute=[@n8.then, @n4.then]
    Tokens:Map of Strings to Strings=@n2
    AC:Animatable Creature Object Reference=@n1.AC File
    __WorldContext:Object Reference
   ) -> (then, Source Path:String, Level Sequence Path:String, Animation Path:String)
   > n6 = ResolveArtifactByName!(execute, Asset Name:String=Get loc_AnimName() -> (loc_AnimName:String), artifact_name:String=anim_sequence, __WorldContext:Object Reference) -> (then, ReturnValue:String)
   > n7 = ResolveArtifactByName(Asset Name=Get loc_AnimName(), artifact_name=level_sequence)
   > Return Node!(
    execute
    Source Path:String=@n5
    Level Sequence Path:String=Append!(A=@n5.Level Sequence Path, B=/, C:String=Split(SourceString=@n7, InStr=/, SearchDir=FromEnd).RightS) -> (ReturnValue)
    Animation Path:String=Append(A=@n5.Animation Path, B=/, C=Split(SourceString=@n6, InStr=/, SearchDir=FromEnd).RightS)
   )
   n8 = Add!(execute=@n4.else, TargetMap:Map of Strings to Strings=@n2, Key:"String (by ref)"=Subfolder, Value:"String (by ref)"=@n1.Feature) -> (then) {is=title K2Node_CallFunction}
  ]}
 Get SKM From AC File {nodes [
   n1 = Get SKM From AC File!() -> (then, AC file:Animatable Creature Object Reference) {is=title K2Node_FunctionEntry}
   > n2 = Cast To AnimatableProp!(execute, Object=@n1) -> (then, CastFailed, AsAnimatable Prop:Animatable Prop Object Reference) {is=title K2Node_DynamicCast}
   > n3 = LoadAsset_Blocking(Asset=Get Skeletal Mesh Asset!(self:Animatable Prop Object Reference=@n2) -> (Skeletal Mesh Asset:Skeletal Mesh Soft Object Reference))
   > n4 = Cast To SkeletalMesh!(execute, Object=@n3) -> (then, CastFailed, AsSkeletal Mesh:Skeletal Mesh Object Reference) {is=title K2Node_DynamicCast}
   > Return Node!(execute, Skeletal Mesh:Skeletal Mesh Object Reference=@n4)
   n5 = Get Sequence Actor from AC File!(execute=[@n2.CastFailed, @n4.CastFailed], AC file:Animatable Creature Object Reference=@n1, __WorldContext:Object Reference) -> (then, Object:Object Reference) {is=title K2Node_CallFunction}
   > n6 = GetBlueprintAsset(Object=@n5)
   > n7 = Gather Subobject Data For Blueprint(self=Get SubobjectDataSubsystem(), Context=@n6)
   > n8 = For Each Loop(Array:Array of Subobject Data Handle Structures=@n7) -> (Array Element:Subobject Data Handle Structure)
   Return Node(execute=@n8.Completed, Skeletal Mesh=Get NewLocalVar() -> (NewLocalVar:Skeletal Mesh Object Reference))
   > n9 = Cast To SkeletalMeshComponent!(execute, Object=GetAssociatedObject!(Data:"Subobject Data Structure (by ref)"=GetData(DataHandle=@n8)) -> (ReturnValue:Object Reference)) -> (then, CastFailed, AsSkeletal Mesh Component:Skeletal Mesh Component Object Reference) {is=title K2Node_DynamicCast}
   n10 = Get SkeletalMesh(self=@n9)
   > Is Valid(InputObject=@n10)
   > Set NewLocalVar(NewLocalVar:Skeletal Mesh Object Reference=@n10) -> (Output_Get:Skeletal Mesh Object Reference)
  ]}
 Get Sequence Actor from AC File {nodes [
   n1 = Get Sequence Actor from AC File!() -> (then, AC file:Animatable Creature Object Reference) {is=title K2Node_FunctionEntry}
   n2 = "To String (SoftClassReference)"!(SoftClassReference:"Object Soft Class Reference (by ref)"=Get Sequencer Actor(self=@n1)) -> (ReturnValue:String)
   n3 = Split(SourceString=@n2, InStr=., SearchDir=FromEnd)
   > n4 = LoadAsset(AssetPath=SelectString(A=@n3, B=Split(SourceString=@n2, InStr=_C, SearchDir=FromEnd), bPickA=@n3.ReturnValue))
   > Return Node!(execute, Object:Object Reference=@n4)
  ]}
 Open Default Animation Map {nodes [
   n1 = Open Default Animation Map!() -> (then, Project:BP Animation Browser Project Object Reference) {is=title K2Node_FunctionEntry}
   n2 = Get Levels!(self:BP Animation Browser Project Object Reference=@n1) -> (Levels:Array of World Soft Object References)
   n3 = "Get (a copy)"(Array:Array of World Soft Object References=@n2) -> (Output:World Soft Object Reference)
   n4 = Length(TargetArray:Array of World Soft Object References=@n2)
   > n5 = Branch(Condition=AND Boolean(A="integer >= integer"(A=@n4, B=1), B=IsValidSoftObjectReference(SoftObjectReference=@n3)))
   > n6 = Set Level Path(Level Path:String=To Soft Object Path!(SoftObjectReference:Soft Object Reference=@n3) -> (ReturnValue_PathString:String)) -> (Output_Get:String)
   n7 = DoesAssetExist(execute=[@n10.then, @n6.then], AssetPath=Get Level Path() -> (Level Path:String))
   > n8 = Branch(Condition=@n7)
   > n9 = LoadMap!(execute, Filename:String=Get Level Path()) -> (then, ReturnValue:World Object Reference)
   > Return Node!(execute, Map:World Object Reference=@n9)
   Return Node(execute=@n8.else)
   n10 = Set Level Path(
    execute=@n5.else
    Level Path=Combine!(InPaths:Array of Strings=Make Array("[0]":String=/AnimInEdTools/AnimationBrowser/, "[1]":String=Assets/AnimationBrowser_DefaultLevel) -> (Array:Array of Strings)) -> (ReturnValue:String)
   )
  ]}
 Paste Pose {nodes [
   n1 = Paste Pose!() -> (
    then
    Pose Asset:Soft Object Reference
    LevelSequence:Level Sequence Object Reference
    ControlRig Proxy Binding:"Control Rig Sequencer Binding Proxy Structure (by ref)"
   ) {is=title K2Node_FunctionEntry}
   > n2 = LoadAsset_Blocking(Asset=@n1)
   > Is Valid(InputObject=@n2)
   > n3 = Cast To ControlRigPoseAsset!(execute, Object=@n2) -> (then, CastFailed, AsControl Rig Pose Asset:Control Rig Pose Asset Object Reference) {is=title K2Node_DynamicCast}
   > n4 = GetControlRigs(LevelSequence=Get LevelSequence() -> (LevelSequence:Level Sequence Object Reference))
   > Branch(Condition=Is Not Empty!(TargetArray:Array of Control Rig Sequencer Binding Proxy Structures=@n4) -> (ReturnValue:bool))
   n5 = Get ControlRig Proxy Binding() -> (ControlRig Proxy Binding:"Control Rig Sequencer Binding Proxy Structure (by ref)")
   > n6 = Execute Python Script!(
    execute
    PythonScript="import unreal\r\n\r\nnum_keys = 0\r\nfor section in proxy.track.get_sections():\r\n    float_channels = section.get_channels_by_type(unreal.MovieSceneScriptingFloatChannel)\r\n    for channel in float_channels:\r\n        num_keys += channel.get_num_keys()\r\n\r\nresult = num_keys == 0\r\n"
    proxy:"Control Rig Sequencer Binding Proxy Structure (by ref)"=@n5
   ) -> (then, ReturnValue, result:bool=~) {inputs [proxy], outputs [result]}
   > Branch(Condition=@n6.result)
   n7 = Break Control Rig Sequencer Binding Proxy!(ControlRigSequencerBindingProxy=@n5) -> (Proxy:Movie Scene Binding Proxy Structure, ControlRig, Track)
   > SelectControls!(
    execute
    self:Control Rig Pose Asset Object Reference=@n3
    InControlRig:Control Rig Object Reference=@n7.ControlRig
    bDoMirror:bool=false
    bClearSelection:bool=true
   ) -> (then)
   > PastePose!(
    execute
    self:Control Rig Pose Asset Object Reference=@n3
    InControlRig:Control Rig Object Reference=@n7.ControlRig
    bDoKey:bool=true
    bDoMirror:bool=false
    bDoAdditive:bool=false
   ) -> (then)
   > ClearControlSelection!(execute, self:Control Rig Object Reference=@n7.ControlRig, bSetupUndo:bool=false) -> (then, ReturnValue:bool)
   "Get (a copy)"(Array:Array of Control Rig Sequencer Binding Proxy Structures=@n4) -> (Output:Control Rig Sequencer Binding Proxy Structure)
  ]}
 ResolveACPaths_SGEnabled {nodes [
   n1 = ResolveACPaths_SGEnabled!() -> (then, name:String) {is=title K2Node_FunctionEntry}
   > n2 = Execute Python Script!(
    execute
    PythonScript="import sys\r\nimport AnimationBrowser\r\n\r\nfrom importlib import reload\r\nreload(AnimationBrowser)\r\n\r\nAMB = sys.modules['AnimationBrowser']\r\n\r\ntokens = AMB.sg_resolve_ac_paths(name)\r\n\r\nsource_path = tokens[\"Source Path\"]\r\nanim_path = tokens[\"Animation Path\"]"
    name:String=@n1
   ) -> (then, ReturnValue, source_path:String, anim_path:String) {inputs [name], outputs [source_path,anim_path]}
   > Return Node!(execute, source_path:String=@n2.source_path, anim_path:String=@n2.anim_path)
  ]}
 ResolveArtifactByName {nodes [
   n1 = ResolveArtifactByName!() -> (then, Asset Name:String, artifact_name:String=anim_sequence) {is=title K2Node_FunctionEntry}
   > n2 = Execute Python Script!(
    execute
    PythonScript="import epic_shrd_techart.intg.common.scheme.utils as SUTILS\r\n\r\nartifactsDict = SUTILS.resolve_artifacts(asset_name, force_refresh=False)\r\n\r\nout = next(\r\n  (a[\"path\"] for a in artifactsDict[\"artifacts\"] if a[\"name\"] == artifact_name),\r\n  None,\r\n)\r\n"
    asset_name:String=@n1
    artifact_name:String=@n1.artifact_name
   ) -> (then, ReturnValue, out:String) {inputs [asset_name, artifact_name], outputs [out]}
   > Return Node!(execute, ReturnValue:String=@n2.out)
  ]}
 ResolvePathsFromTokens {nodes [
   n1 = ResolvePathsFromTokens!() -> (then, Tokens:Map of Strings to Strings, AC:Animatable Creature Object Reference) {is=title K2Node_FunctionEntry}
   n2 = Get Naming Convention!(self:Animatable Creature Object Reference=@n1.AC) -> (ReturnValue:EU Animation Naming Convention Object Reference) {is=title K2Node_CallFunction}
   > Set Skip Token Formatting!(execute, Skip Token Formatting:bool=true, self:EU Animation Naming Convention Object Reference=@n2) -> (then, Output_Get:bool)
   n3 = Get RemoveSpacesOnName!(self:Animatable Creature Object Reference=@n1.AC) -> (RemoveSpacesOnName:bool)
   n4 = Get Path!(
    self:EU Animation Naming Convention Object Reference=@n2
    Token Map:Map of Strings to Strings=@n1
    Path Pattern:String=Get Source Path!(self:Animatable Creature Object Reference=@n1.AC) -> (Source Path:String)
    Remove Spaces On Name:bool=@n3
   ) -> (File Path:String, Folder Path:String, File Name:String) {is=title K2Node_CallFunction}
   n5 = Split(SourceString=@n4, InStr=/, SearchDir=FromEnd)
   > n6 = Branch(Condition=@n5.ReturnValue)
   n7 = Get Animation Path!(self:Animatable Creature Object Reference=@n1.AC) -> (Animation Path:String)
   > Return Node!(
    execute
    Source Path:String=@n4
    Level Sequence Path:String=@n5
    Animation Path:String=Split(SourceString=Get Path(self=@n2, Token Map=@n1, Path Pattern=@n7, Remove Spaces On Name=@n3), InStr=/, SearchDir=FromEnd)
   )
   PrintString(execute=@n6.else, InString="Error: Invalid Source Path found from ResolvePathsFromTokens()")
  ]}
 ReverseTokenMap {nodes [
   n1 = ReverseTokenMap!() -> (then, Asset Name:String, AC File:Animatable Creature Object Reference) {is=title K2Node_FunctionEntry}
   > n2 = Execute Python Script!(
    execute
    PythonScript="import sys\r\nimport AnimationBrowser\r\n\r\nfrom importlib import reload\r\nreload(AnimationBrowser)\r\n\r\nAMB = sys.modules['AnimationBrowser']\r\n\r\ntokens = AMB.get_asset_name_tokens(nm,ac)\r\nprint(tokens)"
    nm:String=@n1
    ac:Animatable Creature Object Reference=@n1.AC File
   ) -> (then, ReturnValue, tokens:Map of Strings to Strings) {inputs [nm,ac], outputs [tokens]}
   > Return Node!(execute, TokenMap:Map of Strings to Strings=@n2.tokens)
  ]}
]
variables []
