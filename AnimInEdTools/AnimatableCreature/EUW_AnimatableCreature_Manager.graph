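# Editor Utility Widget that manages AnimatableCreature objects for the animation editor tools, exposing a selected creature's details, naming and project settings.
# PreConstruct points a single-property picker and a category-filtered details view at the widget itself, so its variables act as the form.
# Changing the selection copies the creature's settings, such as rig, paths, tags and projects, into those variables.
# Subfolder, variation, descriptor, direction and anim-set copies are unwired, and Save has no handler.
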EditorUtilityWidgetBlueprint /AnimInEdTools/AnimatableCreature/EUW_AnimatableCreature_Manager
graphs [EventGraph {nodes [
   "On Property Changed (SelectedACPropertyView)"!() -> (OutputDelegate, then, PropertyName:Name=None) {is=title K2Node_ComponentBoundEvent}
   n1 = Get Selected Animatable Creature() -> (Selected Animatable Creature:Animatable Creature Object Reference)
   > Set Asset(Asset:String=Get Asset!(self:Animatable Creature Object Reference=@n1) -> (Asset:String)) -> (Output_Get:String)
   > Set Variant(Variant:String=Get Variant!(self:Animatable Creature Object Reference=@n1) -> (Variant:String)) -> (Output_Get:String)
   > Set Animation Picker(Animation Picker:Editor Utility Widget Soft Class Reference=Get Animation Picker!(self:Animatable Creature Object Reference=@n1) -> (Animation Picker:Editor Utility Widget Soft Class Reference)) -> (Output_Get:Editor Utility Widget Soft Class Reference)
   > Set Control Rig(Control Rig:Control Rig Asset Soft Reference Structure=Get Control Rig!(self:Animatable Creature Object Reference=@n1) -> (Control Rig:Control Rig Asset Soft Reference Structure)) -> (Output_Get:Control Rig Asset Soft Reference Structure)
   > Set Sequence Actor(Sequence Actor:Object Soft Class Reference=Get Sequencer Actor!(self:Animatable Creature Object Reference=@n1) -> (Sequencer Actor:Object Soft Class Reference)) -> (Output_Get:Object Soft Class Reference)
   > Set IK FK Matching Data(IK FK Matching Data:PDA Ik Fk Matching Bake Data Soft Object Reference=Get IK FK Matching Data!(self:Animatable Creature Object Reference=@n1) -> (IK FK Matching Data:PDA Ik Fk Matching Bake Data Soft Object Reference)) -> (Output_Get:PDA Ik Fk Matching Bake Data Soft Object Reference)
   > Set Enabled(Enabled:bool=Get Enabled!(self:Animatable Creature Object Reference=@n1) -> (Enabled:bool)) -> (Output_Get:bool)
   > Set Default Pose(Default Pose:Control Rig Pose Asset Soft Object Reference=Get Default Pose!(self:Animatable Creature Object Reference=@n1) -> (Default Pose:Control Rig Pose Asset Soft Object Reference)) -> (Output_Get:Control Rig Pose Asset Soft Object Reference)
   > Set Control Rig Backward Solve Enabled(Control Rig Backward Solve Enabled:bool=Get Control Rig Backward Solve Enabled!(self:Animatable Creature Object Reference=@n1) -> (Control Rig Backward Solve Enabled:bool)) -> (Output_Get:bool)
   > Set Tags(Tags:Gameplay Tag Container Structure=Get Tags!(self:Animatable Creature Object Reference=@n1) -> (Tags:Gameplay Tag Container Structure)) -> (Output_Get:Gameplay Tag Container Structure)
   > Set Animation Path(Animation Path:String=Get Animation Path!(self:Animatable Creature Object Reference=@n1) -> (Animation Path:String)) -> (Output_Get:String)
   > Set Source Path(Source Path:String=Get Source Path!(self:Animatable Creature Object Reference=@n1) -> (Source Path:String)) -> (Output_Get:String)
   > Set AnimMontage Path(AnimMontage Path:String=Get AnimMontage Path!(self:Animatable Creature Object Reference=@n1) -> (AnimMontage Path:String)) -> (Output_Get:String)
   > Set Projects(Projects:Array of BP Animation Browser Project Soft Object References=Get Projects!(self:Animatable Creature Object Reference=@n1) -> (Projects:Array of BP Animation Browser Project Soft Object References)) -> (Output_Get:Array of BP Animation Browser Project Soft Object References)
   Event PreConstruct()
   > SetObject(self=Get ACDetailsView() -> (ACDetailsView:Details View Object Reference), NewObject=Self-Reference())
   > SetObject(self=Get SelectedACPropertyView() -> (SelectedACPropertyView:Single Property View Object Reference), NewObject=Self-Reference())
   Get Anim Set!(self:Animatable Creature Object Reference=@n1) -> (Subfolders:Array of Strings) {is=title K2Node_CallFunction}
   Get Subfolders!(self:Animatable Creature Object Reference=@n1) -> (Subfolders:Array of Strings) {is=title K2Node_CallFunction}
   Get Variations!(self:Animatable Creature Object Reference=@n1) -> (Subfolders:Array of Strings) {is=title K2Node_CallFunction}
   Get Descriptors!(self:Animatable Creature Object Reference=@n1) -> (Subfolders:Array of Strings) {is=title K2Node_CallFunction}
   Get Direction!(self:Animatable Creature Object Reference=@n1) -> (Subfolders:Array of Strings) {is=title K2Node_CallFunction}
   Set Subfolders(Subfolders:Array of BP Animation Browser Subfolder Soft Object References) -> (Output_Get:Array of BP Animation Browser Subfolder Soft Object References)
   Set Variations(Variations:Array of BP Animation Browser Variation Soft Object References) -> (Output_Get:Array of BP Animation Browser Variation Soft Object References)
   Set Descriptors(Descriptors:Array of BP Animation Browser Descriptor Soft Object References) -> (Output_Get:Array of BP Animation Browser Descriptor Soft Object References)
   Set Directions(Directions:Array of BP Animation Browser Direction Soft Object References) -> (Output_Get:Array of BP Animation Browser Direction Soft Object References)
   Set Anim Sets(Anim Sets:Array of BP Animation Browser Anim Set Soft Object References) -> (Output_Get:Array of BP Animation Browser Anim Set Soft Object References)
  ]}]
variables [
 Anim Sets {type="array<BP_AnimationBrowser_AnimSet_C>", default []}
 Animation Path {type=string, default=""}
 Animation Picker {type=Class, default=null}
 AnimMontage Path {type=string, default=""}
 Asset {type=string, default=""}
 Control Rig {type=ControlRigAssetSoftReference, default {blueprintRigClass=null, controlRigAsset=null}}
 Control Rig Backward Solve Enabled {type=boolean}
 Default Pose {type=ControlRigPoseAsset, default=null}
 Descriptors {type="array<BP_AnimationBrowser_Descriptor_C>", default []}
 Directions {type="array<BP_AnimationBrowser_Direction_C>", default []}
 Enabled {type=boolean}
 IK FK Matching Data {type=PDA_IkFkMatchingBakeData_C, default=null}
 Naming Convention {type=Class, default=null}
 Naming Convention Instance {type=EU_AnimationNamingConvention_C, default=null}
 Naming Settings {type=Enum_AC_Naming, default=Guided}
 Projects {type="array<BP_AnimationBrowser_Project_C>", default []}
 Selected Animatable Creature {type=AnimatableCreature_C, default=null}
 Sequence Actor {type=Class, default=null}
 Source Path {type=string, default=""}
 Subfolders {type="array<BP_AnimationBrowser_Subfolder_C>", default []}
 Tags {type=GameplayTagContainer, default {gameplayTags []}}
 Variant {type=string, default=""}
 Variations {type="array<BP_AnimationBrowser_Variation_C>", default []}
]
widgets {
 info {rootWidgetClass=&StackBox, widgetCount=7}
 widgets [
  StackBox_0 {widgetClassPath=&StackBox, widget=StackBox_0}
  VerticalBox_0 {
   widgetClassPath=&VerticalBox
   parent=StackBox_0
   widget=VerticalBox_0
   slot {class=StackBoxSlot, properties {size {sizeRule=Fill, value=1}}}
  }
  SelectedACPropertyView {
   widgetClassPath=&SinglePropertyView
   parent=VerticalBox_0
   widget=SelectedACPropertyView
   properties {nameOverride=Animatable Creature, propertyName=Selected Animatable Creature}
   slot {class=VerticalBoxSlot, properties {padding=<ltrb 0 10 0 10>}}
  }
  VerticalBox_1 {
   widgetClassPath=&VerticalBox
   parent=VerticalBox_0
   widget=VerticalBox_1
   slot {class=VerticalBoxSlot, properties {size {sizeRule=Fill, value=1}}}
  }
  ACDetailsView {
   widgetClassPath=&DetailsView
   parent=VerticalBox_1
   widget=ACDetailsView
   properties {categoriesToShow [Selected AC Details, Selected AC Naming, Selected AC Project]}
   slot {class=VerticalBoxSlot, properties {size {sizeRule=Fill, value=1}}}
  }
  SaveACButton {widgetClassPath=&EditorUtilityButton, parent=VerticalBox_1, widget=SaveACButton, slot {class=VerticalBoxSlot}}
  TextBlock_68 {widgetClassPath=&TextBlock, parent=SaveACButton, widget=TextBlock_68, properties {text=Save}, slot {class=ButtonSlot}}
 ]
}
