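# Editor utility widget that creates a test asset of the AnimatableProp class from a single button, part of the AnimInEdTools animatable creature tooling.
# Clicking the button gets the Asset Tools interface and calls CreateAsset with a hardcoded name, AP_Test, in the plugin's AnimatableCreature folder.
# The construct and tick events are empty, and no factory is supplied to the creation call.
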EditorUtilityWidgetBlueprint /AnimInEdTools/AnimatableCreature/EUW_InstanceMaker
graphs [EventGraph {nodes [
   "On Clicked (EditorUtilityButton_82)"() {is=title K2Node_ComponentBoundEvent}
   > n1 = GetAssetTools()
   > CreateAsset(
    self=@n1
    AssetName=AP_Test
    PackagePath=/AnimInEdTools/AnimatableCreature/
    AssetClass=&/AnimatableProp.AnimatableProp_C
   )
   Event PreConstruct()
   Event Construct()
   Event Tick()
  ]}]
variables []
widgets {
 info {rootWidgetClass=&EditorUtilityButton, widgetCount=1}
 widgets [EditorUtilityButton_82 {widgetClassPath=&EditorUtilityButton, widget=EditorUtilityButton_82, slot=~}]
}
