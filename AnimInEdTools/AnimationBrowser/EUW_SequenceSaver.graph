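# Editor utility widget with one button that re-exports and saves the AnimSequences selected in the Content Browser from their linked Level Sequences.
# Each sequence's link gives the Level Sequence path, resolving redirectors via the DestinationObject tag, and the binding GUID for ExportAnimSequence.
# An inline Python node builds the export options with transforms and morph targets enabled.
# Re-exported assets are collected and saved together after the loop, even if not dirty.
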EditorUtilityWidgetBlueprint /AnimInEdTools/AnimationBrowser/EUW_SequenceSaver
graphs [EventGraph {nodes [
   "On Clicked (Resave_Button)"() {is=title K2Node_ComponentBoundEvent}
   > Clear(TargetArray:Array of Object References=Get Assets to Save() -> (Assets to Save:Array of Object References))
   > n1 = GetSelectedAssetsOfClass(AssetClass=/Script/Engine.AnimSequence)
   > n2 = For Each Loop(Array:Array of Object References=@n1) -> (Array Element:Object Reference)
   SaveLoadedAssets(execute=@n2.Completed, AssetsToSave=Get Assets to Save(), bOnlyIfIsDirty=false)
   n3 = Cast To AnimSequence!(execute=@n2.LoopBody, Object=@n2) -> (then, CastFailed, AsAnim Sequence:Anim Sequence Object Reference)
   > n4 = Set Anim Sequence(Anim Sequence:Anim Sequence Object Reference=@n3) -> (Output_Get:Anim Sequence Object Reference)
   > n5 = GetLevelSequenceLinkFromAnimSequence!(execute, InAnimSequence:Anim Sequence Object Reference=@n4) -> (then, ReturnValue:Anim Sequence Level Sequence Link Object Reference)
   > Is Valid(InputObject=@n5)
   > n6 = Set Level Sequence Link(Level Sequence Link:Anim Sequence Level Sequence Link Object Reference=@n5) -> (Output_Get:Anim Sequence Level Sequence Link Object Reference)
   n7 = GetBaseFilename(InPath=Get!(self:Anim Sequence Level Sequence Link Object Reference=@n6) -> (PathToLevelSequence_PathString:String), bRemovePath=false)
   > n8 = DoesAssetExist(AssetPath=@n7)
   > Branch(Condition=@n8)
   > n9 = FindAssetData!(execute, AssetPath:String=@n7) -> (then, ReturnValue:Asset Data Structure)
   n10 = MakeLiteralName(Value=DestinationObject)
   > n11 = LoadAsset(AssetPath=Select(
     Option 0:String=@n7
     Option 1:String=GetTagValue(InAssetData=@n9, InTagName=@n10)
     Index=IsRedirector!(InAssetData:"Asset Data Structure (by ref)"=@n9) -> (ReturnValue:bool)
    ) -> (ReturnValue:String))
   > n12 = Cast To LevelSequence!(execute, Object=@n11) -> (then, CastFailed, AsLevel Sequence:Level Sequence Object Reference) {is=title K2Node_DynamicCast}
   > Set Level Sequence(Level Sequence:Level Sequence Object Reference=@n12) -> (Output_Get:Level Sequence Object Reference)
   > n13 = Execute Python Script!(
    execute
    PythonScript="import unreal\r\n \r\n## Create animation sequence export options\r\nanim_seq_export_options = unreal.AnimSeqExportOption()\r\nanim_seq_export_options.export_transforms = True\r\nanim_seq_export_options.export_morph_targets = True\r\n \r\n"
   ) -> (then, ReturnValue, anim_seq_export_options:Anim Seq Export Option Object Reference) {outputs [anim_seq_export_options]}
   > n14 = GetEditorWorld(self=Get UnrealEditorSubsystem())
   > n15 = FindBindingById!(
    execute
    Sequence:Movie Scene Sequence Object Reference=Get Level Sequence() -> (Level Sequence:Level Sequence Object Reference)
    BindingId:Guid Structure=Get SkelTrackGuid!(self:Anim Sequence Level Sequence Link Object Reference=Get Level Sequence Link() -> (Level Sequence Link:Anim Sequence Level Sequence Link Object Reference)) -> (SkelTrackGuid:Guid Structure)
   ) -> (then, ReturnValue:Movie Scene Binding Proxy Structure)
   > ExportAnimSequence!(
    execute
    World:World Object Reference=@n14
    Sequence:Level Sequence Object Reference=Get Level Sequence()
    AnimSequence:Anim Sequence Object Reference=Get Anim Sequence() -> (Anim Sequence:Anim Sequence Object Reference)
    ExportOptions:Anim Seq Export Option Object Reference
    Binding:"Movie Scene Binding Proxy Structure (by ref)"=@n15
    bCreateLink:bool=false
    ExportOption:Anim Seq Export Option Object Reference=@n13.anim_seq_export_options
   ) -> (then, ReturnValue:bool)
   > Add(TargetArray:Array of Object References=Get Assets to Save(), NewItem:"Object Reference (by ref)"=Get Anim Sequence()) {is=title K2Node_CallArrayFunction}
   Event PreConstruct()
   Event Construct()
   Event Tick()
  ]}]
variables [
 Anim Sequence {type=AnimSequence, default=null}
 Assets to Save {type="array<Object>", default []}
 Level Sequence {type=LevelSequence, default=null}
 Level Sequence Link {type=AnimSequenceLevelSequenceLink, default=null}
]
widgets {
 info {rootWidgetClass=&StackBox, widgetCount=3}
 widgets [
  StackBox_0 {widgetClassPath=&StackBox, widget=StackBox_0}
  Resave_Button {
   widgetClassPath=&EditorUtilityButton
   parent=StackBox_0
   widget=Resave_Button
   slot {class=StackBoxSlot, properties {size {sizeRule=Fill, value=1}}}
  }
  TextBlock_160 {
   widgetClassPath=&TextBlock
   parent=Resave_Button
   widget=TextBlock_160
   properties {autoWrapText=true, text=Re-save anim sequences selected in Content Browser}
   slot {class=ButtonSlot}
  }
 ]
}
