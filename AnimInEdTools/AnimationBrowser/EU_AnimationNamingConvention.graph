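# Editor utility Blueprint that builds and parses animation asset paths under a token-based naming convention.
# Path building optionally reformats chosen tokens (space removal or camel case), appends the default pattern when none has braces, then fills placeholders via embedded Python, blanking missing tokens and collapsing extra separators.
# Parsing takes the parent folder as subfolder and uses Python to match underscore-split name parts case-insensitively against Animatable Creature lists, flagging leftover ambiguity.
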EditorUtilityBlueprint /AnimInEdTools/AnimationBrowser/EU_AnimationNamingConvention
graphs [
 EventGraph {links [], nodes []}
 Format String {nodes [
   n1 = Format String!() -> (then, Input String:String, Remove Underscores In Name:bool=~) {is=title K2Node_FunctionEntry}
   n2 = Self-Reference()
   > Return Node!(
    execute
    Formatted String:String=Select(
     Option 0:String=Remove Spaces String!(String:String=@n1, __WorldContext:Object Reference=@n2) -> (New String:String)
     Option 1:String=Camel Case String!(String:String=@n1, __WorldContext:Object Reference=@n2) -> (New String:String)
     Index=@n1.Remove Underscores In Name
    ) -> (ReturnValue:String)
   )
  ]}
 Get Path {nodes [
   n1 = Get Path!() -> (then, Token Map:Map of Strings to Strings, Path Pattern:String, Remove Spaces On Name:bool=true) {is=title K2Node_FunctionEntry}
   n2 = Get Path From Pattern!(Token Map:Map of Strings to Strings=@n1, Path Pattern:String=@n1.Path Pattern, Remove Spaces On Name:bool=@n1.Remove Spaces On Name) -> (File Path:String, Folder Path:String, File Name:String) {is=title K2Node_CallFunction}
   > Return Node!(execute, File Path:String=@n2, Folder Path:String=@n2.Folder Path, File Name:String=@n2.File Name)
  ]}
 Get Path From Pattern {nodes [
   n1 = Get Path From Pattern!() -> (then, Token Map:Map of Strings to Strings, Path Pattern:String, Remove Spaces On Name:bool=~) {is=title K2Node_FunctionEntry}
   n2 = Execute Python Script!(
    execute
    PythonScript="output = path_pattern.format(**token_map)\r\n"
    token_map:Map of Strings to Strings=@n1
    path_pattern:String=@n1.Path Pattern
    remove_spaces:bool=@n1.Remove Spaces On Name
   ) -> (then, ReturnValue, output:Wildcard) {inputs [token_map, path_pattern, remove_spaces], outputs [output]}
   n3 = Make Array!("[0]":String=Get Path Pattern() -> (Path Pattern:String), "[1]":String=Get Default Path Pattern() -> (Default Path Pattern:String)) -> (Array:Array of Strings)
   n4 = Replace Tokens With Token Values!(
    execute=@n7.then_1
    self
    Token Map:Map of Strings to Strings=Get Token Map() -> (Token Map:Map of Strings to Strings)
    String:String=SelectString(
     A=Get Path Pattern()
     B=Combine!(InPaths:Array of Strings=@n3) -> (ReturnValue:String)
     bPickA=Contains(SearchIn=@n1.Path Pattern, Substring="{")
    )
   ) -> (then, Output String:String)
   Return Node(
    execute=[@n4.then, @n2.then]
    File Path=@n4
    Folder Path=GetPath!(InPath:String=@n4) -> (ReturnValue:String)
    File Name=GetBaseFilename(InPath=@n4, bRemovePath=true)
   )
   n5 = For Each Loop(Exec=@n7.then_0, Array=Get Tokens to Format() -> (Tokens to Format:Array of Strings))
   n6 = Find(TargetMap=Get Token Map(), Key=@n5)
   Add!(
    execute=@n8.else
    TargetMap:Map of Strings to Strings=Get Token Map()
    Key:"String (by ref)"=@n5
    Value:"String (by ref)"=Format String!(Input String:String=@n6, Remove Underscores In Name:bool=Get Remove Spaces On Name() -> (Remove Spaces On Name:bool)) -> (Formatted String:String) {is=title K2Node_CallFunction}
   ) -> (then) {is=title K2Node_CallFunction}
   Add(execute=@n8.then, TargetMap=Get Token Map(), Key=@n5, Value=@n6)
   n7 = Sequence(execute=@n1.then)
   n8 = Branch(execute=@n9.then, Condition=Get Skip Token Formatting() -> (Skip Token Formatting:bool))
   n9 = Branch(execute=@n5.LoopBody, Condition=@n6.ReturnValue)
  ]}
 Get Subfolder Root {nodes [
   n1 = Get Subfolder Root!() -> (then, Path:String) {is=title K2Node_FunctionEntry}
   > Return Node!(execute, Root Path:String=Split(SourceString=@n1, InStr="/{Subfolder}", SearchDir=FromStart))
  ]}
 Get Tokens from Path {nodes [
   Get Tokens from Path!() -> (then, Anim Sequence Path:String, Animatable Creature:Animatable Creature Object Reference) {is=title K2Node_FunctionEntry}
   > n1 = Sequence!(execute) -> (then_0, then_1, then_2:Exec, then_3:Exec, then_4:Exec)
   n2 = Get Animatable Creature() -> (Animatable Creature:Animatable Creature Object Reference)
   n3 = Execute Python Script!(
    execute=@n1.then_2
    PythonScript="anim_set = ''\r\nname = ''\r\ndescriptor = ''\r\ndirection = ''\r\nvariation = ''\r\n\r\nvariations_lower = [x.lower() for x in variations]\r\ndescriptors_lower = [x.lower() for x in descriptors]\r\ndirections_lower = [x.lower() for x in directions]\r\nanim_sets_lower = [x.lower() for x in anim_sets]\r\n\r\n## Split by '_'\r\nsections = file_name.split('_')\r\n\r\nif len(sections) and sections[0] == asset:\r\n    sections = sections[1:] \r\n\r\nif len(sections) and sections[0] == plugin:\r\n    sections = sections[1:] \r\n\r\nremaining_sections = sections\r\n\r\nfor section in sections:\r\n    section_lower = section.lower()\r\n\r\n    if not variation:\r\n        try:\r\n            i = variations_lower.index(section_lower)\r\n            variation = variations[i]\r\n            remaining_sections.remove(section)\r\n            continue\r\n        except ValueError:\r\n            pass\r\n    if not direction:\r\n        try:\r\n            i = directions_lower.index(section_lower)\r\n            direction = directions[i]\r\n            remaining_sections.remove(section)\r\n            continue\r\n        except ValueError:\r\n            pass\r\n    if not descriptor:\r\n        try:\r\n            i = descriptors_lower.index(section_lower)\r\n            descriptor = descriptors[i]\r\n            remaining_sections.remove(section)\r\n            continue\r\n        except ValueError:\r\n            pass\r\n    if not anim_set:\r\n        try:\r\n            i = anim_sets_lower.index(section_lower)\r\n            anim_set = anim_sets[i]\r\n            remaining_sections.remove(section)\r\n            continue\r\n        except ValueError:\r\n            pass\r\n\r\nname = remaining_sections[-1]\r\nsuccess = len(remaining_sections) <= 1"
    asset:String=Get Asset!(self:Animatable Creature Object Reference=@n2) -> (Asset:String)
    plugin:String=Get Variant!(self:Animatable Creature Object Reference=@n2) -> (Variant:String)
    file_name:String=GetBaseFilename(InPath=Get Anim Sequence Path() -> (Anim Sequence Path:String), bRemovePath=true)
    variations:Array of Strings=Get Variations!(self:Animatable Creature Object Reference=Get Animatable Creature()) -> (Subfolders:Array of Strings) {is=title K2Node_CallFunction}
    directions:Array of Strings=Get Direction!(self:Animatable Creature Object Reference=Get Animatable Creature()) -> (Subfolders:Array of Strings) {is=title K2Node_CallFunction}
    descriptors:Array of Strings=Get Descriptors!(self:Animatable Creature Object Reference=Get Animatable Creature()) -> (Subfolders:Array of Strings) {is=title K2Node_CallFunction}
    anim_sets:Array of Strings=Get Anim Set!(self:Animatable Creature Object Reference=Get Animatable Creature()) -> (Subfolders:Array of Strings) {is=title K2Node_CallFunction}
   ) -> (then, ReturnValue, anim_set:String, name:String, descriptor:String, direction:String, variation:String, Success:bool=~) {inputs [asset,plugin,file_name,variations,directions,descriptors,anim_sets], outputs [anim_set,name,descriptor,direction,variation,Success]}
   > Return Node!(
    execute
    Subfolder:String=Get Subfolder Local() -> (Subfolder Local:String)
    Anim Set:String=@n3.anim_set
    Name:String=@n3.name
    Descriptor:String=@n3.descriptor
    Direction:String=@n3.direction
    Variation:String=@n3.variation
    Success:bool=@n3.Success
   )
   n4 = Get Anim Sequence Path()
   Set Subfolder Local(execute=@n1.then_0, Subfolder Local:String=GetBaseFilename(InPath=GetPath(InPath=@n4), bRemovePath=true)) -> (Output_Get:String)
  ]}
 Replace Tokens With Token Values {nodes [
   n1 = Replace Tokens With Token Values!() -> (then, Token Map:Map of Strings to Strings, String:String) {is=title K2Node_FunctionEntry}
   > Set Current String(Current String:String=@n1.String) -> (Output_Get:String)
   > Keys!(execute, TargetMap:Map of Strings to Strings=Get Token Map()) -> (then, Keys:Array of Strings)
   > n2 = Execute Python Script!(
    execute
    PythonScript="from string import Formatter\r\nimport re\r\n\r\nclass KeyFormatter(Formatter):\r\n    def get_value(self, key, args, kwds):\r\n        if isinstance(key, str):\r\n            try:\r\n                return kwds[key]\r\n            except KeyError:\r\n                return ''\r\n        else:\r\n            return Formatter.get_value(key, args, kwds)\r\n\r\nkey_formatter = KeyFormatter()\r\nout_string = key_formatter.format(in_string, **token_dict)\r\nout_string = re.sub('__+', '_', out_string).strip('_')\r\nout_string = re.sub('//+', '/', out_string)\r\n\r\n'''\r\nprint('in_string', in_string)\r\nprint('out_string', out_string)\r\nprint('token_dict', token_dict)\r\nprint('\\n')\r\n'''"
    in_string:String=Get Current String() -> (Current String:String)
    token_dict:Map of Strings to Strings=@n1
   ) -> (then, ReturnValue, out_string:String) {inputs [in_string,token_dict], outputs [out_string]}
   > Return Node!(execute, Output String:String=@n2.out_string)
  ]}
]
variables [
 Default Path Pattern {type=string, default="{Subfolder}/{Asset}_{AnimationSet}_{AnimationName}_{Descriptor}_{Direction}_{Variation}"}
 File Part Separator {type=string, default=_}
 Skip Token Formatting {type=boolean}
 Tokens to Format {type="array<string>", default [AnimationName, Descriptor]}
]
