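# Editor utility widget that batch-offsets float key values on named Control Rig channels across a list of Level Sequences.
# A details view edits the sequence list and channel-name-to-offset map, persisted through a SaveGame slot, and a button adds sequences selected in the Content Browser.
# Run uses Sequencer scripting to find each name's first matching channel per sequence and adds the offset to its keys.
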EditorUtilityWidgetBlueprint /AnimInEdTools/AnimationChannelInspector/EUW_AnimationChannelEditor
graphs [
 EventGraph {nodes [
   "On Released (Run_btn)"() {is=title K2Node_ComponentBoundEvent}
   > n1 = For Each Loop(Array:Array of Level Sequence Soft Object References=Get Sequences() -> (Sequences:Array of Level Sequence Soft Object References)) -> (Array Element:Level Sequence Soft Object Reference)
   > PrintString(InString=Append!(
     A="Checking Level Sequence : "
     B=GetSoftPtrAssetName!(Asset:Soft Object Reference=@n1) -> (ReturnValue:String)
     C:String=" -----------------------"
    ) -> (ReturnValue))
   > Sequence!(execute) -> (then_0, then_1, then_2:Exec, then_3:Exec, then_4:Exec)
   n2 = Get Float Channels Offset() -> (Float Channels Offset:"Map of Names to Float (double-precision)s")
   > n3 = Keys!(execute, TargetMap:"Map of Names to Float (double-precision)s"=@n2) -> (then, Keys:Array of Names)
   > n4 = Values!(execute, TargetMap:"Map of Names to Float (double-precision)s"=@n2) -> (then, Values:"Array of Float (double-precision)s")
   > n5 = For Each Loop(Array:Array of Names=@n3) -> (Array Element:Name)
   > n6 = Get Control Rig Channel by Type and Name!(
    execute
    self
    Level Sequence:Level Sequence Soft Object Reference=@n1
    Channel Type:Movie Scene Scripting Channel Class Reference=/Script/SequencerScripting.MovieSceneScriptingFloatChannel
    Channel Name:Name=@n5
   ) -> (then, Array Element:Movie Scene Scripting Channel Object Reference) {is=title K2Node_CallFunction}
   > Offset Float Channel!(
    execute
    self
    NewParam:Movie Scene Scripting Channel Object Reference=@n6
    Offset:double="Get (a ref)"(Array:"Array of Float (double-precision)s"=@n4, Dimension 1=@n5.Array Index) -> (Output:"Float (double-precision) (by ref)")
   ) -> (then, Result:bool)
   "On Released (GetFromContentBrowser_btn)"() {is=title K2Node_ComponentBoundEvent}
   > n7 = GetSelectedAssets()
   > n8 = For Each Loop(Array:Array of Object References=@n7) -> (Array Element:Object Reference)
   > n9 = Cast To LevelSequence!(execute, Object=@n8) -> (then, CastFailed, AsLevel Sequence:Level Sequence Object Reference) {is=title K2Node_DynamicCast}
   n10 = Make Soft Reference!(Input:Level Sequence Object Reference=@n9) -> (Output:Level Sequence Soft Object Reference)
   n11 = Get Sequences()
   > Branch(Condition="integer < integer"(A=Find Item!(TargetArray:Array of Level Sequence Soft Object References=@n11, ItemToFind:"Level Sequence Soft Object Reference (by ref)"=@n10) -> (ReturnValue)))
   > Add(TargetArray:Array of Level Sequence Soft Object References=Get Sequences(), NewItem:"Level Sequence Soft Object Reference (by ref)"=@n10) {is=title K2Node_CallArrayFunction}
   "On Property Changed (Input_dv)"!() -> (OutputDelegate, then, PropertyName:Name=None) {is=title K2Node_ComponentBoundEvent}
   n12 = Get SaveGame() -> (SaveGame:SG Animation Channel Inspector Object Reference)
   > Set Sequences!(execute, Sequences:Array of Level Sequence Soft Object References=Get Sequences(), self:SG Animation Channel Inspector Object Reference=@n12) -> (then, Output_Get:Array of Level Sequence Soft Object References)
   > Set Float Channels Edit!(
    execute
    Float Channels Edit:"Map of Names to Float (double-precision)s"=Get Float Channels Offset()
    self:SG Animation Channel Inspector Object Reference=@n12
   ) -> (then, Output_Get:"Map of Names to Float (double-precision)s")
   > AsyncSaveGameToSlot!(
    execute:Exec
    WorldContextObject:Object Reference
    SaveGameObject:Save Game Object Reference=@n12
    SlotName:String=AnimationChannelEditor
    UserIndex:int=0
   ) -> (then:Exec, Completed:Exec, SaveGame:Save Game Object Reference, bSuccess:bool=~) {is=title K2Node_AsyncAction}
   Event PreConstruct()
   > SetObject(self=Get Input_dv() -> (Input_dv:Details View Object Reference), NewObject=Self-Reference())
   Event Construct()
   > n13 = LoadGameFromSlot!(execute, SlotName:String=AnimationChannelEditor, UserIndex:int=0) -> (then, ReturnValue:Save Game Object Reference)
   > n14 = Cast To SG_AnimationChannelInspector!(execute, Object=@n13) -> (then, CastFailed, AsSG Animation Channel Inspector:SG Animation Channel Inspector Object Reference) {is=title K2Node_DynamicCast}
   > Set SaveGame(SaveGame:SG Animation Channel Inspector Object Reference=@n14) -> (Output_Get:SG Animation Channel Inspector Object Reference)
   > Set Sequences!(execute, Sequences=Get Sequences!(self:SG Animation Channel Inspector Object Reference=@n14) -> (Sequences)) -> (then, Output_Get)
   > Set Float Channels Offset(Float Channels Offset:"Map of Names to Float (double-precision)s"=Get Float Channels Edit!(self:SG Animation Channel Inspector Object Reference=@n14) -> (Float Channels Edit:"Map of Names to Float (double-precision)s")) -> (Output_Get:"Map of Names to Float (double-precision)s")
   n15 = CreateSaveGameObject!(
    execute=@n14.CastFailed
    SaveGameClass:Save Game Class Reference=&/SG_AnimationChannelInspector.SG_AnimationChannelInspector_C
   ) -> (then, ReturnValue:SG Animation Channel Inspector Object Reference)
   > Set SaveGame(SaveGame=@n15)
   Event Tick()
  ]}
 Get Control Rig Channel by Type and Name {nodes [
   n1 = Get Control Rig Channel by Type and Name!() -> (then, Level Sequence:Level Sequence Soft Object Reference, Channel Type:Movie Scene Scripting Channel Class Reference, Channel Name:Name) {is=title K2Node_FunctionEntry}
   > n2 = LoadAsset_Blocking(Asset=@n1)
   > n3 = Cast To LevelSequence(Object=@n2)
   > n4 = GetControlRigs!(execute, LevelSequence:Level Sequence Object Reference=@n3) -> (then, ReturnValue:Array of Control Rig Sequencer Binding Proxy Structures)
   > n5 = For Each Loop(Array:Array of Control Rig Sequencer Binding Proxy Structures=@n4) -> (Array Element:Control Rig Sequencer Binding Proxy Structure)
   Return Node!(execute=@n5.Completed, Array Element:Movie Scene Scripting Channel Object Reference)
   n6 = Break Control Rig Sequencer Binding Proxy!(ControlRigSequencerBindingProxy:"Control Rig Sequencer Binding Proxy Structure (by ref)"=@n5) -> (
    Proxy:Movie Scene Binding Proxy Structure
    ControlRig:Control Rig Object Reference
    Track:Movie Scene Control Rig Parameter Track Object Reference
   ) {is=title K2Node_BreakStruct}
   > n7 = GetTracks!(execute, InBinding:"Movie Scene Binding Proxy Structure (by ref)"=@n6) -> (then, ReturnValue:Array of Movie Scene Track Object References)
   > n8 = For Each Loop(Array:Array of Movie Scene Track Object References=@n7) -> (Array Element:Movie Scene Track Object Reference)
   > n9 = GetSections!(execute, Track:Movie Scene Track Object Reference=@n8) -> (then, ReturnValue:Array of Movie Scene Section Object References)
   > n10 = For Each Loop(Array:Array of Movie Scene Section Object References=@n9) -> (Array Element:Movie Scene Section Object Reference)
   > n11 = GetChannelsByType!(
    execute
    Section:Movie Scene Section Object Reference=@n10
    ChannelType:Movie Scene Scripting Channel Class Reference=Get Channel Type() -> (Channel Type:Movie Scene Scripting Channel Class Reference)
   ) -> (then, ReturnValue:Array of Movie Scene Scripting Channel Object References)
   > n12 = For Each Loop with Break!(Exec:Exec, Array:Array of Movie Scene Scripting Channel Object References=@n11, Break:Exec) -> (LoopBody:Exec, Array Element:Movie Scene Scripting Channel Object Reference, Array Index:int, Completed:Exec)
   > Branch(Condition="Equal (Name)"(
     A=Get ChannelName!(self:Movie Scene Scripting Channel Object Reference=@n12) -> (ChannelName:Name=None)
     B=Get Channel Name() -> (Channel Name:Name=None)
    ))
   n13 = Get Channel Name() -> (Channel Name=None)
   > PrintString(InString=Append!(A="Found Channel (", B="To String (Name)"(InName=@n13), C=") on Rig : ", D:String=GetDisplayName(Object=@n6.ControlRig), E:String) -> (ReturnValue))
   > Return Node(Array Element=@n12)
  ]}
 Offset Float Channel {nodes [
   n1 = Offset Float Channel!() -> (then, NewParam:Movie Scene Scripting Channel Object Reference, Offset:double=~) {is=title K2Node_FunctionEntry}
   > n2 = Cast To MovieSceneScriptingFloatChannel!(execute, Object=@n1) -> (then, CastFailed, AsMovie Scene Scripting Float Channel:Movie Scene Scripting Float Channel Object Reference) {is=title K2Node_DynamicCast}
   > n3 = For Each Loop with Break(Array:Array of Movie Scene Scripting Key Object References="Get Keys (Float)"!(self:Movie Scene Scripting Float Channel Object Reference=@n2) -> (ReturnValue:Array of Movie Scene Scripting Key Object References) {is=title K2Node_CallFunction}) -> (Array Element:Movie Scene Scripting Key Object Reference)
   > n4 = Cast To MovieSceneScriptingFloatKey!(execute, Object=@n3) -> (then, CastFailed, AsMovie Scene Scripting Float Key:Movie Scene Scripting Float Key Object Reference) {is=title K2Node_DynamicCast}
   > "Set Value (Float)"!(
    execute
    self:Movie Scene Scripting Float Key Object Reference=@n4
    InNewValue:float=float + float(
     A="Get Value (Float)"!(self:Movie Scene Scripting Float Key Object Reference=@n4) -> (ReturnValue:float) {is=title K2Node_CallFunction}
     B=Get Offset() -> (Offset:double)
    )
   ) -> (then) {is=title K2Node_CallFunction}
   Return Node!(execute, Result:bool=true)
  ]}
]
variables [
 Float Channels Offset {type="map<number>", default {}}
 Found {type="array<boolean>", default []}
 NewVar {type=number}
 SaveGame {type=SG_AnimationChannelInspector_C, default=null}
 Sequences {type="array<LevelSequence>", default []}
]
widgets {
 info {rootWidgetClass=&VerticalBox, widgetCount=8}
 widgets [
  VerticalBox_0 {widgetClassPath=&VerticalBox, widget=VerticalBox_0, properties {clipping=ClipToBounds}}
  GetFromContentBrowser_btn {widgetClassPath=&EditorUtilityButton, parent=VerticalBox_0, widget=GetFromContentBrowser_btn, slot {class=VerticalBoxSlot}}
  TextBlock_1 {
   widgetClassPath=&TextBlock
   parent=GetFromContentBrowser_btn
   widget=TextBlock_1
   properties {text=Get from Content Browser}
   slot {class=ButtonSlot}
  }
  EditorUtilityScrollBox_215 {
   widgetClassPath=&EditorUtilityScrollBox
   parent=VerticalBox_0
   widget=EditorUtilityScrollBox_215
   properties {alwaysShowScrollbar=true}
   slot {class=VerticalBoxSlot, properties {size {sizeRule=Fill, value=1}}}
  }
  Input_dv {
   widgetClassPath=&DetailsView
   parent=EditorUtilityScrollBox_215
   widget=Input_dv
   properties {categoriesToShow [InputSequences]}
   slot {class=ScrollBoxSlot}
  }
  Run_btn {widgetClassPath=&EditorUtilityButton, parent=VerticalBox_0, widget=Run_btn, slot {class=VerticalBoxSlot}}
  TextBlock_0 {widgetClassPath=&TextBlock, parent=Run_btn, widget=TextBlock_0, properties {text=Run}, slot {class=ButtonSlot}}
  EditorUtilityScrollBox_0 {
   widgetClassPath=&EditorUtilityScrollBox
   parent=VerticalBox_0
   widget=EditorUtilityScrollBox_0
   properties {alwaysShowScrollbar=true}
   slot {class=VerticalBoxSlot, properties {size {sizeRule=Fill, value=1}}}
  }
 ]
}
