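# Editor utility widget that searches Level Sequences for Control Rig animation keys matching bool, integer and float channel criteria.
# Search loads each sequence, finds named channels on its rigs, tests key values or ranges, and combines results by Any/All.
# Details views expose inputs and matches, buttons exchange assets with the Content Browser, and a SaveGame slot persists settings.
# A separate Audit prints mouth_pose usage tallies and flags out-of-bounds values.
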EditorUtilityWidgetBlueprint /AnimInEdTools/AnimationChannelInspector/EUW_AnimationChannelInspector
graphs [
 AuditMouthShapes {nodes [
   n1 = AuditMouthShapes!() -> (then, ShapeCount:int=48) {is=title K2Node_FunctionEntry}
   > Set LocalShapeCount(LocalShapeCount:int=@n1) -> (Output_Get:int=0)
   n2 = Get Sequences() -> (Sequences:Array of Level Sequence Soft Object References)
   > PrintText(InText=Format Text!(
     Format="LOCTEXT(\"Auditing {sequence_count} Sequences : \\r\\n -----------------------\\r\\n\")"
     sequence_count:int=Length(TargetArray:Array of Level Sequence Soft Object References=@n2)
    ) -> (Result))
   n3 = Get AuditList() -> (AuditList:Array of Integers)
   > Clear(TargetArray:Array of Integers=@n3)
   n4 = Get LocalShapeCount() -> (LocalShapeCount:int=0)
   > Resize!(execute, TargetArray:Array of Integers=@n3, Size:int=@n4) -> (then) {is=title K2Node_CallArrayFunction}
   n5 = Get TempCounter() -> (TempCounter:Array of Integers)
   > Clear(TargetArray=@n5)
   > Resize(TargetArray=@n5, Size=@n4)
   > n6 = For Each Loop(Array:Array of Level Sequence Soft Object References=Get Sequences()) -> (Array Element:Level Sequence Soft Object Reference)
   n7 = For Each Loop(Exec=@n6.Completed, Array:Array of Integers=Get AuditList()) -> (Array Element:int)
   > PrintText(InText=Format Text!(
     Format="LOCTEXT(\"mouth_pose {pose_number} is used {count} times.\")"
     pose_number:Text="To Text (Integer)"(Value=@n7.Array Index, bUseGrouping=false, MinimumIntegralDigits=2, MaximumIntegralDigits=2)
     count:int=@n7
    ) -> (Result))
   n8 = Get Control Rig Channel by Type and Name!(
    execute=@n6.LoopBody
    self
    Level Sequence:Level Sequence Soft Object Reference=@n6
    Channel Type:Movie Scene Scripting Channel Class Reference=/Script/SequencerScripting.MovieSceneScriptingIntegerChannel
    Channel Name:Name=mouth_pose
    Log:bool=false
   ) -> (then, Array Element:Movie Scene Scripting Channel Object Reference) {is=title K2Node_CallFunction}
   > n9 = Cast To MovieSceneScriptingIntegerChannel!(execute, Object=@n8) -> (then, CastFailed, AsMovie Scene Scripting Integer Channel:Movie Scene Scripting Integer Channel Object Reference) {is=title K2Node_DynamicCast}
   > n10 = For Each Loop with Break!(
    Exec:Exec
    Array:Array of Movie Scene Scripting Key Object References="Get Keys (Integer)"!(self:Movie Scene Scripting Integer Channel Object Reference=@n9) -> (ReturnValue:Array of Movie Scene Scripting Key Object References) {is=title K2Node_CallFunction}
    Break:Exec
   ) -> (LoopBody:Exec, Array Element:Movie Scene Scripting Key Object Reference, Array Index:int, Completed:Exec)
   n11 = For Each Loop(Exec=@n10.Completed, Array=Get TempCounter())
   n12 = Get TempCounter()
   Clear(execute=@n11.Completed, TargetArray=@n12)
   > Resize(TargetArray=@n12, Size=Get LocalShapeCount() -> (LocalShapeCount=0))
   Branch(execute=@n11.LoopBody, Condition="integer >= integer"(A=@n11, B=1))
   n13 = Get AuditList()
   > Set Array Elem!(
    execute
    TargetArray:Array of Integers=@n13
    Index:int=@n11.Array Index
    Item:"Integer (by ref)"=int + int(A="Get (a ref)"(Array=@n13, Dimension 1=0))
    bSizeToFit:bool=false
   ) -> (then)
   n14 = Cast To MovieSceneScriptingIntegerKey!(execute=@n10.LoopBody, Object=@n10) -> (then, CastFailed, AsMovie Scene Scripting Integer Key:Movie Scene Scripting Integer Key Object Reference) {is=title K2Node_DynamicCast}
   n15 = "Get Value (Integer)"!(self:Movie Scene Scripting Integer Key Object Reference=@n14) -> (ReturnValue:int=0) {is=title K2Node_CallFunction}
   n16 = Get LocalShapeCount() -> (LocalShapeCount=0)
   > n17 = Branch(Condition=NOR Boolean!(A:bool="integer < integer"(A=@n15), B:bool="integer > integer"(A=@n15, B=@n16)) -> (ReturnValue:bool))
   n18 = Get TempCounter()
   > Set Array Elem(TargetArray=@n18, Index=@n15, Item=int + int(A="Get (a ref)"(Array=@n18, Dimension 1=@n15)), bSizeToFit=false)
   PrintText(
    execute=@n17.else
    InText=Format Text!(
     Format="LOCTEXT(\"{sequence} contains out of bounds key value : {value}\")"
     sequence:String=GetSoftPtrAssetName!(Asset:Soft Object Reference=@n6) -> (ReturnValue:String)
     value:int=@n15
    ) -> (Result)
   )
  ]}
 Bool All {nodes [
   n1 = Bool All!() -> (then, Array:Array of Booleans) {is=title K2Node_FunctionEntry}
   > n2 = For Each Loop with Break(Array:Array of Booleans=@n1) -> (Array Element:bool)
   Return Node!(execute=@n2.Completed, Result:bool=true)
   > n3 = Branch(Condition=@n2)
   Return Node(execute=@n3.else, Result=false)
  ]}
 Bool Any Or All {nodes [
   n1 = Bool Any Or All!() -> (then, Array:Array of Booleans, Any:bool=~) {is=title K2Node_FunctionEntry}
   > n2 = For Each Loop with Break(Array=@n1)
   Return Node(execute=@n2.Completed, Result=NOT Boolean(A=Get Any() -> (Any:bool)))
   > n3 = Branch(Condition=@n2)
   n4 = Get Any()
   > Branch(Condition=@n4)
   > Return Node(Result=true)
   Branch(execute=@n3.else, Condition=NOT Boolean(A=@n4))
   > Return Node(Result=false)
  ]}
 Bool Channel Match Value {nodes [
   n1 = Bool Channel Match Value!() -> (then, NewParam:Movie Scene Scripting Channel Object Reference, Value:bool=~) {is=title K2Node_FunctionEntry}
   > n2 = Cast To MovieSceneScriptingBoolChannel!(execute, Object=@n1) -> (then, CastFailed, AsMovie Scene Scripting Bool Channel:Movie Scene Scripting Bool Channel Object Reference) {is=title K2Node_DynamicCast}
   > n3 = For Each Loop with Break(Array:Array of Movie Scene Scripting Key Object References="Get Keys (Bool)"!(self:Movie Scene Scripting Bool Channel Object Reference=@n2) -> (ReturnValue:Array of Movie Scene Scripting Key Object References) {is=title K2Node_CallFunction}) -> (Array Element:Movie Scene Scripting Key Object Reference)
   PrintString(execute=@n3.Completed, InString=Key not Found)
   > Return Node(Result=false)
   n4 = Cast To MovieSceneScriptingBoolKey!(execute=@n3.LoopBody, Object=@n3) -> (then, CastFailed, AsMovie Scene Scripting Bool Key:Movie Scene Scripting Bool Key Object Reference) {is=title K2Node_DynamicCast}
   > Branch(Condition="Equal (Boolean)"(
     A:bool="Get Value (Bool)"!(self:Movie Scene Scripting Bool Key Object Reference=@n4) -> (ReturnValue:bool) {is=title K2Node_CallFunction}
     B:bool=@n1.Value
    ) -> (ReturnValue:bool=~) {is=title K2Node_PromotableOperator})
   n5 = FrameNumber to Integer!(InFrameNumber:"Frame Number Structure (by ref)"="Get Time (Bool)"!(self:Movie Scene Scripting Bool Key Object Reference=@n4, TimeUnit:EMovieSceneTimeUnit Enum=DisplayRate) -> (ReturnValue_FrameNumber:Frame Number Structure, ReturnValue_SubFrame:float=~) {is=title K2Node_CallFunction}) -> (ReturnValue:int=0)
   > PrintString(InString=Append!(A="Found Key at Frame : ", B="To String (Integer)"(InInt=@n5), C:String) -> (ReturnValue))
   > Return Node(Result=true)
  ]}
 EventGraph {nodes [
   "On Released (Search_btn)"() {is=title K2Node_ComponentBoundEvent}
   > PrintString(InString=Hello)
   > Clear(TargetArray:Array of Level Sequence Soft Object References=Get Output Sequences() -> (Output Sequences:Array of Level Sequence Soft Object References))
   > n1 = For Each Loop(Array:Array of Level Sequence Soft Object References=Get Sequences()) -> (Array Element:Level Sequence Soft Object Reference)
   > PrintString(InString=Append(A="Checking Level Sequence : ", B=GetSoftPtrAssetName(Asset=@n1), C=" -----------------------"))
   > Clear(TargetArray:Array of Booleans=Get Found() -> (Found:Array of Booleans))
   > n2 = Sequence!(execute) -> (then_0, then_1, then_2:Exec, then_3:Exec, then_4:Exec)
   n3 = Get Bool Channels() -> (Bool Channels:Map of Names to Booleans)
   > n4 = Values!(execute, TargetMap:Map of Names to Booleans=@n3) -> (then, Values:Array of Booleans)
   > n5 = Keys!(execute, TargetMap:Map of Names to Booleans=@n3) -> (then, Keys:Array of Names)
   > n6 = For Each Loop(Array:Array of Names=@n5) -> (Array Element:Name)
   > n7 = Get Control Rig Channel by Type and Name(Level Sequence=@n1, Channel Type=/Script/SequencerScripting.MovieSceneScriptingBoolChannel, Channel Name=@n6, Log=true)
   > n8 = Bool Channel Match Value!(
    execute
    self
    NewParam:Movie Scene Scripting Channel Object Reference=@n7
    Value:bool="Get (a ref)"(Array:Array of Booleans=@n4, Dimension 1=@n6.Array Index) -> (Output:"Boolean (by ref)")
   ) -> (then, Result:bool) {is=title K2Node_CallFunction}
   > Add(TargetArray:Array of Booleans=Get Found(), NewItem:"Boolean (by ref)"=@n8) {is=title K2Node_CallArrayFunction}
   n9 = Get Integer Channels() -> (Integer Channels:Map of Names to STRUCT Integer Min Max Structures)
   n10 = Values(execute=@n2.then_1, TargetMap:Map of Names to STRUCT Integer Min Max Structures=@n9) -> (Values:Array of STRUCT Integer Min Max Structures)
   > n11 = Keys(TargetMap:Map of Names to STRUCT Integer Min Max Structures=@n9)
   > n12 = For Each Loop(Array=@n11)
   > n13 = Get Control Rig Channel by Type and Name(Level Sequence=@n1, Channel Type=&MovieSceneScriptingIntegerChannel, Channel Name=@n12, Log=true)
   n14 = Break STRUCT Integer Min Max!(STRUCT_IntegerMinMax:"STRUCT Integer Min Max Structure (by ref)"="Get (a ref)"(Array:Array of STRUCT Integer Min Max Structures=@n10, Dimension 1=@n12.Array Index) -> (Output:"STRUCT Integer Min Max Structure (by ref)")) -> (
    Min_2_#1:int
    Max_4_#2:int
    OutsideRange_8_#3:bool=~
   ) {is=title K2Node_BreakStruct}
   > n15 = Integer Channel Match Range!(
    execute
    self
    NewParam:Movie Scene Scripting Channel Object Reference=@n13
    Min:int=@n14
    Max:int=@n14.Max_4_#2
    Outside Range:bool=@n14.OutsideRange_8_#3
   ) -> (then, Result:bool)
   > Add(TargetArray=Get Found(), NewItem=@n15)
   n16 = Get Float Channels() -> (Float Channels:Map of Names to STRUCT Float Min Max Structures)
   n17 = Values(execute=@n2.then_2, TargetMap:Map of Names to STRUCT Float Min Max Structures=@n16) -> (Values:Array of STRUCT Float Min Max Structures)
   > n18 = Keys(TargetMap:Map of Names to STRUCT Float Min Max Structures=@n16)
   > n19 = For Each Loop(Array=@n18)
   > n20 = Get Control Rig Channel by Type and Name(Level Sequence=@n1, Channel Type=/Script/SequencerScripting.MovieSceneScriptingDoubleChannel, Channel Name=@n19, Log=true)
   n21 = Break STRUCT Float Min Max!(STRUCT_FloatMinMax:"STRUCT Float Min Max Structure (by ref)"="Get (a ref)"(Array:Array of STRUCT Float Min Max Structures=@n17, Dimension 1=@n19.Array Index) -> (Output:"STRUCT Float Min Max Structure (by ref)")) -> (
    Min_8_#1:double=~
    Max_9_#2:double=~
    OutsideRange_12_#4:bool=~
   ) {is=title K2Node_BreakStruct}
   > n22 = Float Channel Match Range!(
    execute
    self
    NewParam:Movie Scene Scripting Channel Object Reference=@n20
    Min:double=@n21
    Max:double=@n21.Max_9_#2
    Outside Range:bool=@n21.OutsideRange_12_#4
   ) -> (then, Result:bool)
   > Add(TargetArray=Get Found(), NewItem=@n22)
   n23 = Bool Any Or All!(execute=@n2.then_4, self, Array:Array of Booleans=Get Found(), Any:bool=Get Any()) -> (then, Result:bool) {is=title K2Node_CallFunction}
   > Branch(Condition=@n23)
   > Add(
    TargetArray:Array of Level Sequence Soft Object References=Get Output Sequences()
    NewItem:"Level Sequence Soft Object Reference (by ref)"=@n1
   )
   "On Released (GetFromContentBrowser_btn)"() {is=title K2Node_ComponentBoundEvent}
   > n24 = GetSelectedAssets()
   > n25 = For Each Loop(Array:Array of Object References=@n24) -> (Array Element:Object Reference)
   > n26 = Cast To LevelSequence!(execute, Object=@n25) -> (then, CastFailed, AsLevel Sequence:Level Sequence Object Reference) {is=title K2Node_DynamicCast}
   n27 = Make Soft Reference!(Input:Level Sequence Object Reference=@n26) -> (Output:Level Sequence Soft Object Reference)
   n28 = Get Sequences()
   > Branch(Condition="integer < integer"(A=Find Item!(TargetArray:Array of Level Sequence Soft Object References=@n28, ItemToFind:"Level Sequence Soft Object Reference (by ref)"=@n27) -> (ReturnValue)))
   > Add(TargetArray=Get Sequences(), NewItem=@n27)
   "On Released (SelectInContentBrowser_btn)"() {is=title K2Node_ComponentBoundEvent}
   > n29 = For Each Loop(Array:Array of Level Sequence Soft Object References=Get Output Sequences()) -> (Array Element:Level Sequence Soft Object Reference)
   > Branch(Condition=IsValidSoftObjectReference(SoftObjectReference=@n29))
   > SyncBrowserToObjects(AssetPaths=Make Array("[0]":String=To Soft Object Path!(SoftObjectReference:Soft Object Reference=@n29) -> (ReturnValue_PathString:String)) -> (Array:Array of Strings))
   "On Property Changed (Input_dv)"!() -> (OutputDelegate, then, PropertyName:Name=None) {is=title K2Node_ComponentBoundEvent}
   n30 = Get SaveGame() -> (SaveGame:SG Animation Channel Inspector Object Reference)
   > Set Sequences!(execute, Sequences:Array of Level Sequence Soft Object References=Get Sequences(), self:SG Animation Channel Inspector Object Reference=@n30) -> (then, Output_Get:Array of Level Sequence Soft Object References)
   > Set Bool Channels!(execute, Bool Channels:Map of Names to Booleans=Get Bool Channels(), self:SG Animation Channel Inspector Object Reference=@n30) -> (then, Output_Get:Map of Names to Booleans)
   > Set Integer Channels!(
    execute
    Integer Channels:Map of Names to STRUCT Integer Min Max Structures=Get Integer Channels()
    self:SG Animation Channel Inspector Object Reference=@n30
   ) -> (then, Output_Get:Map of Names to STRUCT Integer Min Max Structures)
   > Set Float Channels!(
    execute
    Float Channels:Map of Names to STRUCT Float Min Max Structures=Get Float Channels()
    self:SG Animation Channel Inspector Object Reference=@n30
   ) -> (then, Output_Get:Map of Names to STRUCT Float Min Max Structures)
   > Set Any!(execute, Any:bool=Get Any(), self:SG Animation Channel Inspector Object Reference=@n30) -> (then, Output_Get:bool)
   > AsyncSaveGameToSlot!(
    execute:Exec
    WorldContextObject:Object Reference
    SaveGameObject:Save Game Object Reference=@n30
    SlotName:String=AnimationChannelInspector
    UserIndex:int=0
   ) -> (then:Exec, Completed:Exec, SaveGame:Save Game Object Reference, bSuccess:bool=~) {is=title K2Node_AsyncAction}
   "On Clicked (Audit_btn)"() {is=title K2Node_ComponentBoundEvent}
   > AuditMouthShapes!(execute, self, ShapeCount:int=48) -> (then) {is=title K2Node_CallFunction}
   Event PreConstruct()
   > SetObject(
    self=[Get Input_dv() -> (Input_dv:Details View Object Reference), Get Output_dv() -> (Output_dv:Details View Object Reference)]
    NewObject=Self-Reference()
   )
   Event Construct()
   > n31 = LoadGameFromSlot!(execute, SlotName:String=AnimationChannelInspector, UserIndex:int=0) -> (then, ReturnValue:Save Game Object Reference)
   > n32 = Cast To SG_AnimationChannelInspector!(execute, Object=@n31) -> (then, CastFailed, AsSG Animation Channel Inspector:SG Animation Channel Inspector Object Reference) {is=title K2Node_DynamicCast}
   > Set SaveGame(SaveGame:SG Animation Channel Inspector Object Reference=@n32) -> (Output_Get:SG Animation Channel Inspector Object Reference)
   > Set Sequences!(execute, Sequences=Get Sequences!(self:SG Animation Channel Inspector Object Reference=@n32) -> (Sequences)) -> (then, Output_Get)
   > Set Bool Channels!(execute, Bool Channels=Get Bool Channels!(self:SG Animation Channel Inspector Object Reference=@n32) -> (Bool Channels)) -> (then, Output_Get)
   > Set Integer Channels!(execute, Integer Channels=Get Integer Channels!(self:SG Animation Channel Inspector Object Reference=@n32) -> (Integer Channels)) -> (then, Output_Get)
   > Set Float Channels!(execute, Float Channels=Get Float Channels!(self:SG Animation Channel Inspector Object Reference=@n32) -> (Float Channels)) -> (then, Output_Get)
   n33 = CreateSaveGameObject!(
    execute=@n32.CastFailed
    SaveGameClass:Save Game Class Reference=&/SG_AnimationChannelInspector.SG_AnimationChannelInspector_C
   ) -> (then, ReturnValue:SG Animation Channel Inspector Object Reference)
   > Set SaveGame(SaveGame=@n33)
   Event Tick()
   Set Any!(execute, Any=Get Any!(self:SG Animation Channel Inspector Object Reference=@n32) -> (Any)) -> (then, Output_Get)
  ]}
 Float Channel Match Range {nodes [
   n1 = Float Channel Match Range!() -> (then, NewParam:Movie Scene Scripting Channel Object Reference, Min:double=~, Max:double=~, Outside Range:bool=~) {is=title K2Node_FunctionEntry}
   > n2 = Cast To MovieSceneScriptingFloatChannel!(execute, Object=@n1) -> (then, CastFailed, AsMovie Scene Scripting Float Channel:Movie Scene Scripting Float Channel Object Reference) {is=title K2Node_DynamicCast}
   > n3 = For Each Loop with Break(Array="Get Keys (Float)"!(self:Movie Scene Scripting Float Channel Object Reference=@n2) -> (ReturnValue:Array of Movie Scene Scripting Key Object References) {is=title K2Node_CallFunction})
   PrintString(execute=@n3.Completed, InString=Key not Found)
   > Return Node(Result=false)
   n4 = Cast To MovieSceneScriptingFloatKey!(execute=@n3.LoopBody, Object=@n3) -> (then, CastFailed, AsMovie Scene Scripting Float Key:Movie Scene Scripting Float Key Object Reference) {is=title K2Node_DynamicCast}
   > n5 = Branch(Condition=Get Outside Range() -> (Outside Range:bool))
   n6 = "Get Value (Float)"!(self:Movie Scene Scripting Float Key Object Reference=@n4) -> (ReturnValue:float) {is=title K2Node_CallFunction}
   n7 = Get Min() -> (Min:double)
   n8 = Get Max() -> (Max:double)
   > n9 = Branch(Condition=OR Boolean(A="float < float"(A=@n6, B=@n7), B="float > float"(A=@n6, B=@n8)))
   n10 = FrameNumber to Integer(InFrameNumber="Get Time (Float)"!(self:Movie Scene Scripting Float Key Object Reference=@n4, TimeUnit:EMovieSceneTimeUnit Enum=DisplayRate) -> (ReturnValue_FrameNumber:Frame Number Structure, ReturnValue_SubFrame:float=~) {is=title K2Node_CallFunction}) -> (ReturnValue=0)
   PrintString(execute=[@n9.then, @n11.then], InString=Append(A="Found Key at Frame : ", B="To String (Integer)"(InInt=@n10)))
   > Return Node(Result=true)
   n11 = Branch(execute=@n5.else, Condition=AND Boolean(A="float >= float"(A=@n6, B=@n7), B="float <= float"(A=@n6, B=@n8) -> (ReturnValue=~)))
   PrintString(execute=@n4.CastFailed, InString=Failed to Cast)
  ]}
 Get Control Rig Channel by Type and Name {nodes [
   n1 = Get Control Rig Channel by Type and Name!() -> (
    then
    Level Sequence:Level Sequence Soft Object Reference
    Channel Type:Movie Scene Scripting Channel Class Reference
    Channel Name:Name
    Log:bool=true
   ) {is=title K2Node_FunctionEntry}
   > n2 = LoadAsset_Blocking(Asset=@n1)
   > n3 = Cast To LevelSequence(Object=@n2)
   > n4 = GetControlRigs!(execute, LevelSequence:Level Sequence Object Reference=@n3) -> (then, ReturnValue:Array of Control Rig Sequencer Binding Proxy Structures)
   > n5 = For Each Loop(Array:Array of Control Rig Sequencer Binding Proxy Structures=@n4) -> (Array Element:Control Rig Sequencer Binding Proxy Structure)
   Return Node!(execute=@n5.Completed, Array Element:Movie Scene Scripting Channel Object Reference)
   n6 = Break Control Rig Sequencer Binding Proxy!(ControlRigSequencerBindingProxy:"Control Rig Sequencer Binding Proxy Structure (by ref)"=@n5) -> (
    Proxy:Movie Scene Binding Proxy Structure
    ControlRig:Control Rig Object Reference
    Track:Movie Scene Control Rig Parameter Track Object Reference
   ) {is=title K2Node_BreakStruct}
   > n7 = GetSections!(execute, Track:Movie Scene Track Object Reference=@n6.Track) -> (then, ReturnValue:Array of Movie Scene Section Object References)
   > n8 = For Each Loop(Array:Array of Movie Scene Section Object References=@n7) -> (Array Element:Movie Scene Section Object Reference)
   > n9 = GetChannel!(execute, Section:Movie Scene Section Object Reference=@n8, ChannelName:"Name (by ref)"=Get Channel Name() -> (Channel Name:Name=None)) -> (then, ReturnValue:Movie Scene Scripting Channel Object Reference)
   > Branch(Condition=IsValid(Object=@n9))
   n10 = Get Channel Name() -> (Channel Name=None)
   > PrintString(
    InString=Append!(A="Found Channel (", B="To String (Name)"(InName=@n10), C=") on Rig : ", D:String=GetDisplayName(Object=@n6.ControlRig), E:String) -> (ReturnValue)
    bPrintToScreen=@n1.Log
    bPrintToLog=@n1.Log
   )
   > Return Node(Array Element=@n9)
  ]}
 Integer Channel Match Range {nodes [
   n1 = Integer Channel Match Range!() -> (then, NewParam:Movie Scene Scripting Channel Object Reference, Min:int, Max:int, Outside Range:bool=~) {is=title K2Node_FunctionEntry}
   > n2 = Cast To MovieSceneScriptingIntegerChannel(Object=@n1)
   > n3 = For Each Loop with Break(Array="Get Keys (Integer)"(self=@n2))
   PrintString(execute=@n3.Completed, InString=Key not Found)
   > Return Node!(execute, Result:bool=false)
   n4 = Cast To MovieSceneScriptingIntegerKey(execute=@n3.LoopBody, Object=@n3)
   > n5 = Branch(Condition=Get Outside Range())
   n6 = "Get Value (Integer)"(self=@n4) -> (ReturnValue=0)
   n7 = Get Min() -> (Min:int=0)
   n8 = Get Max() -> (Max:int=0)
   > n9 = Branch(Condition=OR Boolean(A="integer < integer"(A=@n6, B=@n7), B="integer > integer"(A=@n6, B=@n8)))
   n10 = FrameNumber to Integer(InFrameNumber="Get Time (Integer)"!(self:Movie Scene Scripting Integer Key Object Reference=@n4, TimeUnit:EMovieSceneTimeUnit Enum=DisplayRate) -> (ReturnValue_FrameNumber:Frame Number Structure, ReturnValue_SubFrame:float=~) {is=title K2Node_CallFunction}) -> (ReturnValue=0)
   PrintString(execute=[@n9.then, @n11.then], InString=Append!(A="Found Key at Frame : ", B="To String (Integer)"(InInt=@n10), C) -> (ReturnValue))
   > Return Node(Result=true)
   n11 = Branch(execute=@n5.else, Condition=AND Boolean(A="integer >= integer"(A=@n6, B=@n7), B="integer <= integer"(A=@n6, B=@n8)))
  ]}
 Integer Channel Match Value {nodes [
   n1 = Integer Channel Match Value!() -> (then, NewParam:Movie Scene Scripting Channel Object Reference, Value:int) {is=title K2Node_FunctionEntry}
   > n2 = Cast To MovieSceneScriptingIntegerChannel(Object=@n1)
   > n3 = For Each Loop with Break(Array="Get Keys (Integer)"(self=@n2))
   PrintString(execute=@n3.Completed, InString=Key not Found)
   > Return Node(Result=false)
   n4 = Cast To MovieSceneScriptingIntegerKey(execute=@n3.LoopBody, Object=@n3)
   > Branch(Condition="Equal (Integer)"(A="Get Value (Integer)"(self=@n4) -> (ReturnValue=0), B=@n1.Value))
   n5 = FrameNumber to Integer(InFrameNumber="Get Time (Integer)"(self=@n4, TimeUnit=DisplayRate) -> (ReturnValue_SubFrame=~)) -> (ReturnValue=0)
   > PrintString(InString=Append(A="Found Key at Frame : ", B="To String (Integer)"(InInt=@n5)))
   > Return Node(Result=true)
  ]}
]
variables [
 Any {type=boolean}
 Bool Channels {type="map<boolean>", default {}}
 Float Channels {type="map<STRUCT_FloatMinMax>", default {}}
 Found {type="array<boolean>", default []}
 Integer Channels {type="map<STRUCT_IntegerMinMax>", default {}}
 Output Sequences {type="array<LevelSequence>", default []}
 SaveGame {type=SG_AnimationChannelInspector_C, default=null}
 Sequences {type="array<LevelSequence>", default []}
]
widgets {
 info {rootWidgetClass=&VerticalBox, widgetCount=13}
 widgets [
  VerticalBox_0 {widgetClassPath=&VerticalBox, widget=VerticalBox_0, properties {clipping=ClipToBounds}}
  GetFromContentBrowser_btn {widgetClassPath=&EditorUtilityButton, parent=VerticalBox_0, widget=GetFromContentBrowser_btn, slot {class=VerticalBoxSlot}}
  TextBlock_1 {
   widgetClassPath=&TextBlock
   parent=GetFromContentBrowser_btn
   widget=TextBlock_1
   properties {text=Get from Content Browser}
   slot {class=ButtonSlot}
  }
  EditorUtilityScrollBox_215 {
   widgetClassPath=&EditorUtilityScrollBox
   parent=VerticalBox_0
   widget=EditorUtilityScrollBox_215
   properties {alwaysShowScrollbar=true}
   slot {class=VerticalBoxSlot, properties {size {sizeRule=Fill, value=1}}}
  }
  Input_dv {
   widgetClassPath=&DetailsView
   parent=EditorUtilityScrollBox_215
   widget=Input_dv
   properties {categoriesToShow [InputSequences]}
   slot {class=ScrollBoxSlot}
  }
  Search_btn {widgetClassPath=&EditorUtilityButton, parent=VerticalBox_0, widget=Search_btn, slot {class=VerticalBoxSlot}}
  TextBlock_0 {widgetClassPath=&TextBlock, parent=Search_btn, widget=TextBlock_0, properties {text=Search}, slot {class=ButtonSlot}}
  EditorUtilityScrollBox_0 {
   widgetClassPath=&EditorUtilityScrollBox
   parent=VerticalBox_0
   widget=EditorUtilityScrollBox_0
   properties {alwaysShowScrollbar=true}
   slot {class=VerticalBoxSlot, properties {size {sizeRule=Fill, value=1}}}
  }
  Output_dv {
   widgetClassPath=&DetailsView
   parent=EditorUtilityScrollBox_0
   widget=Output_dv
   properties {categoriesToShow [OutputSequences]}
   slot {class=ScrollBoxSlot}
  }
  SelectInContentBrowser_btn {widgetClassPath=&EditorUtilityButton, parent=VerticalBox_0, widget=SelectInContentBrowser_btn, slot {class=VerticalBoxSlot}}
  TextBlock_2 {
   widgetClassPath=&TextBlock
   parent=SelectInContentBrowser_btn
   widget=TextBlock_2
   properties {text=Select in Content Browser}
   slot {class=ButtonSlot}
  }
  Audit_btn {widgetClassPath=&EditorUtilityButton, parent=VerticalBox_0, widget=Audit_btn, slot {class=VerticalBoxSlot}}
  TextBlock {widgetClassPath=&TextBlock, parent=Audit_btn, widget=TextBlock, properties {text=AUDIT}, slot {class=ButtonSlot}}
 ]
}
