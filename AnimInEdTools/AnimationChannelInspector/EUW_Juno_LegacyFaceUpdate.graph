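# Editor utility widget that finds and converts legacy Juno face animation (brow poses, mouth transforms) in Control Rig Level Sequences.
# Search flags sequences with keyed poses and an enabled legacy bool, marking animated flags unfixable.
# Fix confirms on the focused sequence, remaps brow poses, scrubs keyed frames to rebake mouth controls from rig curves and per-pose offsets, then disables the flags.
# A SaveGame persists widget inputs between sessions.
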EditorUtilityWidgetBlueprint /AnimInEdTools/AnimationChannelInspector/EUW_Juno_LegacyFaceUpdate
graphs [
 Check Legacy {nodes [
   n1 = Check Legacy!() -> (then, Level Sequence:Level Sequence Soft Object Reference, Pose Channels:Array of Names, Legacy Channel:Name) {is=title K2Node_FunctionEntry}
   > n2 = For Each Loop(Array:Array of Names=@n1.Pose Channels) -> (Array Element:Name)
   Return Node!(execute=@n2.Completed, Use Legacy:bool=false, Can't Fix:bool=false)
   > n3 = Get Control Rig Channel by Type and Name!(
    execute
    self
    Level Sequence:Level Sequence Soft Object Reference=@n1
    Channel Type:Movie Scene Scripting Channel Class Reference=/Script/SequencerScripting.MovieSceneScriptingIntegerChannel
    Channel Name:Name=@n2
    Log:bool=true
   ) -> (then, Array Element:Movie Scene Scripting Channel Object Reference) {is=title K2Node_CallFunction}
   > n4 = "Has Keys (Integer Channel)"!(execute, self, Object:Object Reference=@n3) -> (then, ReturnValue:bool)
   > Branch(Condition=@n4)
   > n5 = Get Control Rig Channel by Type and Name(Level Sequence=@n1, Channel Type=/Script/SequencerScripting.MovieSceneScriptingBoolChannel, Channel Name=@n1.Legacy Channel, Log=true)
   > n6 = "Match Value (Bool Channel)"!(execute, self, NewParam:Movie Scene Scripting Channel Object Reference=@n5, Value:bool=true) -> (then, Result:bool)
   > n7 = Branch(Condition=@n6)
   > n8 = "Has Constant Value (Bool Channel )"!(execute, self, Channel:Movie Scene Scripting Channel Object Reference=@n5) -> (then, Constant:bool)
   > Return Node(Use Legacy=true, Can't Fix=NOT Boolean(A=@n8))
   Return Node(execute=@n7.else, Use Legacy=false, Can't Fix=false)
  ]}
 EventGraph {nodes [
   n1 = LoadAsset_Blocking(Asset=For Each Loop(Array:Array of Level Sequence Soft Object References=Get Legacy Brow Poses Sequences() -> (Legacy Brow Poses Sequences:Array of Level Sequence Soft Object References)) -> (Array Element:Level Sequence Soft Object Reference))
   > n2 = Cast To LevelSequence!(execute, Object=@n1) -> (then, CastFailed, AsLevel Sequence:Level Sequence Object Reference) {is=title K2Node_DynamicCast}
   n3 = "Get KeyValue Array (Bool Channel)"!(
    execute
    self
    Level Sequence:Level Sequence Soft Object Reference=Get Current Sequence() -> (Current Sequence:Level Sequence Soft Object Reference)
    Channel Name:Name=mouth_pose
   ) -> (then, Key Map1:Map of Integers to Integers) {is=title K2Node_CallFunction}
   > Set Mouth Pose Key Map(Mouth Pose Key Map:Map of Integers to Integers=@n3) -> (Output_Get:Map of Integers to Integers)
   "On Released (Search_btn)"() {is=title K2Node_ComponentBoundEvent}
   > Clear(TargetArray:Array of Level Sequence Soft Object References=Get Legacy Brow Poses Sequences())
   > Clear(TargetArray=Get Legacy Face Transforms Sequences() -> (Legacy Face Transforms Sequences:Array of Level Sequence Soft Object References))
   > n4 = For Each Loop(Array=Get Sequences() -> (Sequences:Array of Level Sequence Soft Object References))
   > Set Current Sequence(Current Sequence:Level Sequence Soft Object Reference=@n4) -> (Output_Get:Level Sequence Soft Object Reference)
   > PrintString(InString=Append!(
     A="Checking Level Sequence : "
     B=GetSoftPtrAssetName!(Asset:Soft Object Reference=@n4) -> (ReturnValue:String)
     C:String=" -----------------------"
    ) -> (ReturnValue))
   > n5 = Sequence!(execute) -> (then_0, then_1, then_2:Exec, then_3:Exec, then_4:Exec)
   n6 = Get Current Sequence()
   > n7 = Check Legacy!(
    execute
    self
    Level Sequence:Level Sequence Soft Object Reference=@n6
    Pose Channels:Array of Names=Make Array!("[0]":Name=brow_l_pose, "[1]":Name=brow_r_pose) -> (Array:Array of Names)
    Legacy Channel:Name=Use Legacy Brow Poses
   ) -> (then, Use Legacy:bool, Can't Fix:bool) {is=title K2Node_CallFunction}
   > Branch(Condition=@n7)
   > n8 = Branch(Condition=@n7.Can't Fix)
   > Add(
    TargetArray:Array of Level Sequence Soft Object References=Get Can't Fix Sequences() -> (Can't Fix Sequences:Array of Level Sequence Soft Object References)
    NewItem:"Level Sequence Soft Object Reference (by ref)"=Get Current Sequence()
   ) {is=title K2Node_CallArrayFunction}
   Add(execute=@n8.else, TargetArray=Get Legacy Brow Poses Sequences(), NewItem=Get Current Sequence())
   n9 = Check Legacy(execute=@n5.then_1, Level Sequence=@n6, Pose Channels=Make Array!("[0]"=mouth_pose) -> (Array), Legacy Channel=Use Legacy Face Transforms)
   > Branch(Condition=@n9)
   > n10 = Branch(Condition=@n9.Can't Fix)
   > Add(TargetArray=Get Can't Fix Sequences(), NewItem=Get Current Sequence())
   Add(execute=@n10.else, TargetArray=Get Legacy Face Transforms Sequences(), NewItem=Get Current Sequence())
   "On Released (GetFromContentBrowser_btn)"() {is=title K2Node_ComponentBoundEvent}
   > n11 = GetSelectedAssets()
   > n12 = For Each Loop(Array:Array of Object References=@n11) -> (Array Element:Object Reference)
   > n13 = Cast To LevelSequence(Object=@n12)
   n14 = Make Soft Reference!(Input:Level Sequence Object Reference=@n13) -> (Output:Level Sequence Soft Object Reference)
   n15 = Get Sequences()
   > Branch(Condition="integer < integer"(A=Find Item!(TargetArray:Array of Level Sequence Soft Object References=@n15, ItemToFind:"Level Sequence Soft Object Reference (by ref)"=@n14) -> (ReturnValue)))
   > Add(TargetArray=Get Sequences(), NewItem=@n14)
   "On Released (SelectInContentBrowser_btn)"() {is=title K2Node_ComponentBoundEvent}
   > n16 = For Each Loop(Array:Array of Level Sequence Soft Object References=Get Legacy Brow Poses Sequences()) -> (Array Element:Level Sequence Soft Object Reference)
   > Branch(Condition=IsValidSoftObjectReference(SoftObjectReference=@n16))
   > SyncBrowserToObjects(AssetPaths=Make Array("[0]":String=To Soft Object Path!(SoftObjectReference:Soft Object Reference=@n16) -> (ReturnValue_PathString:String)) -> (Array:Array of Strings))
   "On Property Changed (Input_dv)"!() -> (OutputDelegate, then, PropertyName:Name=None) {is=title K2Node_ComponentBoundEvent}
   n17 = Get SaveGame() -> (SaveGame:SG Animation Channel Inspector Object Reference)
   > Set Sequences!(execute, Sequences:Array of Level Sequence Soft Object References=Get Sequences(), self:SG Animation Channel Inspector Object Reference=@n17) -> (then, Output_Get:Array of Level Sequence Soft Object References)
   > Set Bool Channels!(
    execute
    Bool Channels:Map of Names to Booleans=Get Bool Channels() -> (Bool Channels:Map of Names to Booleans)
    self:SG Animation Channel Inspector Object Reference=@n17
   ) -> (then, Output_Get:Map of Names to Booleans)
   > Set Integer Channels!(
    execute
    Integer Channels:Map of Names to STRUCT Integer Min Max Structures=Get Integer Channels() -> (Integer Channels:Map of Names to STRUCT Integer Min Max Structures)
    self:SG Animation Channel Inspector Object Reference=@n17
   ) -> (then, Output_Get:Map of Names to STRUCT Integer Min Max Structures)
   > Set Float Channels!(
    execute
    Float Channels:Map of Names to STRUCT Float Min Max Structures=Get Float Channels() -> (Float Channels:Map of Names to STRUCT Float Min Max Structures)
    self:SG Animation Channel Inspector Object Reference=@n17
   ) -> (then, Output_Get:Map of Names to STRUCT Float Min Max Structures)
   > Set Any!(execute, Any:bool=Get Any() -> (Any:bool), self:SG Animation Channel Inspector Object Reference=@n17) -> (then, Output_Get:bool)
   > AsyncSaveGameToSlot!(
    execute:Exec
    WorldContextObject:Object Reference
    SaveGameObject:Save Game Object Reference=@n17
    SlotName:String=AnimationChannelInspector
    UserIndex:int=0
   ) -> (then:Exec, Completed:Exec, SaveGame:Save Game Object Reference, bSuccess:bool=~) {is=title K2Node_AsyncAction}
   "On Clicked (Fix_btn)"() {is=title K2Node_ComponentBoundEvent}
   > n18 = GetFocusedLevelSequence!(execute) -> (then, ReturnValue:Level Sequence Object Reference)
   > Set Current Sequence(Current Sequence=Make Soft Reference(Input=@n18))
   > n19 = Sequence!(execute) -> (then_0, then_1, then_2)
   > Set Run Fix Mouth(Run Fix Mouth:bool=true) -> (Output_Get:bool)
   n20 = Get Current Sequence()
   > n21 = Check Legacy(
    Level Sequence=@n20
    Pose Channels=Make Array!("[0]":Name=brow_l_pose, "[1]":Name=brow_r_pose) -> (Array:Array of Names)
    Legacy Channel=Use Legacy Brow Poses
   )
   > n22 = Check Legacy(Level Sequence=@n20, Pose Channels=Make Array!("[0]"=mouth_pose) -> (Array), Legacy Channel=Use Legacy Face Transforms)
   > n23 = Branch(Condition=OR Boolean(A=@n21.Can't Fix, B=@n22.Can't Fix))
   n24 = MakeLiteralText(Value="LOCTEXT(\"Warning\")")
   > Show Message Dialog(
    Title=@n24
    Message=MakeLiteralText(Value="LOCTEXT(\"Can\\'t Fix Animated Legacy Option\")")
    MessageType=Ok
    DefaultValue=No
    MessageCategory=Warning
   )
   n25 = Select(Option 0:String, Option 1:String=Fix Legacy Mouth Offset, Index=AND Boolean(A=@n22, B=Get Execute Mouth Fix() -> (Execute Mouth Fix:bool))) -> (ReturnValue:String)
   n26 = Select(Option 1="Fix Legacy Brow Poses\r\n", Index=AND Boolean(A=Get Execute Brows Fix() -> (Execute Brows Fix:bool), B=@n21))
   n27 = Show Message Dialog(
    execute=@n23.else
    Title=@n24
    Message="To Text (String)"(InString=Append!(A="Actions:\r\n", B=@n26, C=@n25, D:String, E:String) -> (ReturnValue))
    MessageType=OkCancel
    DefaultValue=Cancel
    MessageCategory=Warning
   )
   n28 = "Equal (Byte)"(A:Byte=@n27, B:Byte=5) -> (ReturnValue:bool=~) {is=title K2Node_PromotableOperator}
   > Set Run Fix Mouth(Run Fix Mouth=AND Boolean!(A=@n28, B=@n22, C:bool=Get Execute Mouth Fix()) -> (ReturnValue))
   > Set Run Fix Brows(Run Fix Brows:bool=AND Boolean(A=@n21, B=@n28, C=Get Execute Brows Fix())) -> (Output_Get:bool)
   Branch(execute=@n19.then_1, Condition=Get Run Fix Brows() -> (Run Fix Brows:bool))
   > Fix Brows!(
    execute
    self
    Level Sequence:Level Sequence Soft Object Reference=Get Current Sequence()
    Brow Pose Map:Array of Integers=Get Brow Pose Map() -> (Brow Pose Map:Array of Integers)
   ) -> (then)
   Branch(execute=@n19.then_2, Condition=Get Run Fix Mouth() -> (Run Fix Mouth:bool))
   > Fix Mouth!(
    execute
    self
    Level Sequence:Level Sequence Soft Object Reference=Get Current Sequence()
    Mouth Offsets:Array of Transforms=Get Mouth Pose Offset() -> (Mouth Pose Offset:Array of Transforms)
   ) -> (then)
   Event PreConstruct()
   > SetObject(
    self=[
     Get Options_dv() -> (Options_dv:Details View Object Reference)
     Get Input_dv() -> (Input_dv:Details View Object Reference)
     Get Output_dv() -> (Output_dv:Details View Object Reference)
    ]
    NewObject=Self-Reference()
   )
   Event Construct()
   > n29 = LoadGameFromSlot!(execute, SlotName:String=AnimationChannelInspector, UserIndex:int=0) -> (then, ReturnValue:Save Game Object Reference)
   > n30 = Cast To SG_AnimationChannelInspector!(execute, Object=@n29) -> (then, CastFailed, AsSG Animation Channel Inspector:SG Animation Channel Inspector Object Reference) {is=title K2Node_DynamicCast}
   > Set SaveGame(SaveGame:SG Animation Channel Inspector Object Reference=@n30) -> (Output_Get:SG Animation Channel Inspector Object Reference)
   > Set Sequences!(execute, Sequences=Get Sequences!(self:SG Animation Channel Inspector Object Reference=@n30) -> (Sequences)) -> (then, Output_Get)
   > Set Bool Channels!(execute, Bool Channels=Get Bool Channels!(self:SG Animation Channel Inspector Object Reference=@n30) -> (Bool Channels)) -> (then, Output_Get)
   > Set Integer Channels!(execute, Integer Channels=Get Integer Channels!(self:SG Animation Channel Inspector Object Reference=@n30) -> (Integer Channels)) -> (then, Output_Get)
   > Set Float Channels!(execute, Float Channels=Get Float Channels!(self:SG Animation Channel Inspector Object Reference=@n30) -> (Float Channels)) -> (then, Output_Get)
   n31 = CreateSaveGameObject!(
    execute=@n30.CastFailed
    SaveGameClass:Save Game Class Reference=&/SG_AnimationChannelInspector.SG_AnimationChannelInspector_C
   ) -> (then, ReturnValue:SG Animation Channel Inspector Object Reference)
   > Set SaveGame(SaveGame=@n31)
   n32 = Sequence!(execute) -> (then_0, then_1)
   > For Each Loop(Array=Get Legacy Brow Poses Sequences())
   For Each Loop(Exec=@n32.then_1, Array=Get Legacy Face Transforms Sequences())
   n33 = Sequence!(execute) -> (then_0, then_1, then_2:Exec, then_3:Exec, then_4:Exec)
   n34 = For Each Loop(
    Exec=@n33.then_2
    Array:Array of Movie Scene Scripting Channel Object References=Get Current Location Channels() -> (Current Location Channels:Array of Movie Scene Scripting Channel Object References)
   ) -> (Array Element:Movie Scene Scripting Channel Object Reference)
   > n35 = Cast To MovieSceneScriptingFloatChannel!(execute, Object=@n34) -> (then, CastFailed, AsMovie Scene Scripting Float Channel:Movie Scene Scripting Float Channel Object Reference) {is=title K2Node_DynamicCast}
   > n36 = For Each Loop(Array:Array of Movie Scene Scripting Key Object References="Get Keys (Float)"!(self:Movie Scene Scripting Float Channel Object Reference=@n35) -> (ReturnValue:Array of Movie Scene Scripting Key Object References) {is=title K2Node_CallFunction}) -> (Array Element:Movie Scene Scripting Key Object Reference)
   > n37 = Cast To MovieSceneScriptingFloatKey!(execute, Object=@n36) -> (then, CastFailed, AsMovie Scene Scripting Float Key:Movie Scene Scripting Float Key Object Reference) {is=title K2Node_DynamicCast}
   n38 = "Get Time (Float)"!(self:Movie Scene Scripting Float Key Object Reference=@n37, TimeUnit:EMovieSceneTimeUnit Enum=DisplayRate) -> (ReturnValue:Frame Time Structure) {is=title K2Node_CallFunction}
   > n39 = Get Pose at Frame!(
    execute
    self
    TargetMap:Map of Integers to Integers=Get Mouth Pose Key Map() -> (Mouth Pose Key Map:Map of Integers to Integers)
    Time:int=Break Frame Number!(FrameNumber:"Frame Number Structure (by ref)"=Break Frame Time!(FrameTime:"Frame Time Structure (by ref)"=@n38) -> (FrameNumber:Frame Number Structure, SubFrame:float=~) {is=title K2Node_BreakStruct}) -> (Value:int) {is=title K2Node_BreakStruct}
   ) -> (then, Pose:int=0) {is=title K2Node_CallFunction}
   > n40 = Get Pose Offset Transform!(execute, self, Array:Array of Transforms=Get Mouth Pose Offset(), Pose:int=@n39) -> (then, MouthU:double, MouthV:double, MouthScaleU:double, MouthScaleV:double, Location Offsets:"Array of Float (double-precision)s") {is=title K2Node_CallFunction}
   n41 = "Get Value (Float)"!(self:Movie Scene Scripting Float Key Object Reference=@n37) -> (ReturnValue:float) {is=title K2Node_CallFunction}
   n42 = "Get (a ref)"(Array:"Array of Float (double-precision)s"=@n40.Location Offsets, Dimension 1=@n34.Array Index) -> (Output:"Float (double-precision) (by ref)")
   > "Set Value (Float)"!(execute, self:Movie Scene Scripting Float Key Object Reference=@n37, InNewValue:float=float - float(A=@n41, B=@n42)) -> (then) {is=title K2Node_CallFunction}
   n43 = Get ChannelName!(self:Movie Scene Scripting Channel Object Reference=@n35) -> (ChannelName:Name=None)
   > PrintString(InString=Append(A="To String (Name)"(InName=@n43), B=" - ", C="To String (Integer)"(InInt=@n39), D=" - ", E="To String (Float)"(InDouble=@n42)))
   n44 = For Each Loop(Array:Array of Movie Scene Scripting Channel Object References=Get Current Scale Channels() -> (Current Scale Channels:Array of Movie Scene Scripting Channel Object References)) -> (Array Element:Movie Scene Scripting Channel Object Reference)
   > n45 = Cast To MovieSceneScriptingFloatChannel(Object=@n44)
   > n46 = For Each Loop(Array:Array of Movie Scene Scripting Key Object References="Get Keys (Float)"(self=@n45)) -> (Array Element:Movie Scene Scripting Key Object Reference)
   > n47 = Cast To MovieSceneScriptingFloatKey(Object=@n46)
   n48 = Break Frame Number(FrameNumber=Break Frame Time(FrameTime="Get Time (Float)"(self=@n47, TimeUnit=DisplayRate)) -> (SubFrame=~))
   > n49 = Get Pose at Frame(TargetMap=Get Mouth Pose Key Map(), Time=@n48) -> (Pose=0)
   > n50 = Get Pose Offset Transform!(execute, self, Array=Get Mouth Pose Offset(), Pose=@n49) -> (then, MouthU, MouthV, MouthScaleU, MouthScaleV, Scale Offsets:"Array of Float (double-precision)s")
   n51 = Get ChannelName(self=@n44) -> (ChannelName=None)
   n52 = "Get (a ref)"(Array=@n50.Scale Offsets, Dimension 1=@n44.Array Index)
   > PrintString(InString=Append!(
     A="To String (Name)"(InName=@n51)
     B=" - "
     C="To String (Integer)"(InInt=@n48)
     D=" - "
     E="To String (Integer)"(InInt=@n49)
     F:String=" - "
     G:String="To String (Float)"(InDouble=@n52)
    ) -> (ReturnValue))
   > "Set Value (Float)"(self=@n47, InNewValue=float * float(A="Get Value (Float)"(self=@n47), B=@n52) {is=title K2Node_PromotableOperator})
   OpenLevelSequence!(execute, LevelSequence:Level Sequence Object Reference=@n2) -> (then, ReturnValue:bool)
   Event Tick()
   float + float(A=@n41, B=@n42)
   float / float(B=@n52)
   Set Any!(execute, Any=Get Any!(self:SG Animation Channel Inspector Object Reference=@n30) -> (Any)) -> (then, Output_Get)
  ]}
 Fix Brows {nodes [
   n1 = Fix Brows!() -> (then, Level Sequence:Level Sequence Soft Object Reference, Brow Pose Map:Array of Integers) {is=title K2Node_FunctionEntry}
   > n2 = Sequence!(execute) -> (then_0, then_1, then_2, then_3)
   > PrintString(InString=Append!(A="Fixing Level Sequence : ", B=GetSoftPtrAssetName(Asset=@n1), C=" -----------------------") -> (ReturnValue))
   n3 = Get Control Rig Channel by Type and Name(
    execute=@n2.then_1
    Level Sequence=Get Level Sequence() -> (Level Sequence:Level Sequence Soft Object Reference)
    Channel Type=&MovieSceneScriptingIntegerChannel
    Channel Name=brow_l_pose
    Log=true
   )
   > "Remap Value (Integer Channel)"!(execute, self, NewParam:Movie Scene Scripting Channel Object Reference=@n3, Brow Pose Map:Array of Integers=@n1.Brow Pose Map) -> (then)
   > n4 = Get Control Rig Channel by Type and Name(
    Level Sequence=Get Level Sequence()
    Channel Type=&MovieSceneScriptingIntegerChannel
    Channel Name=brow_r_pose
    Log=true
   )
   > "Remap Value (Integer Channel)"(NewParam=@n4, Brow Pose Map=@n1.Brow Pose Map)
   n5 = Get Control Rig Channel by Type and Name(
    execute=@n2.then_2
    Level Sequence=Get Level Sequence()
    Channel Type=/Script/SequencerScripting.MovieSceneScriptingFloatChannel
    Channel Name=brow_l_ctrl.Scale.Z
    Log=true
   )
   > "Update Value (Float Channel)"!(execute, self, Float Channel:Movie Scene Scripting Channel Object Reference=@n5, Value:double=1) -> (then, Result:bool)
   > n6 = Get Control Rig Channel by Type and Name(
    Level Sequence=Get Level Sequence()
    Channel Type=&MovieSceneScriptingFloatChannel
    Channel Name=brow_r_ctrl.Scale.Z
    Log=true
   )
   > "Update Value (Float Channel)"(Float Channel=@n6, Value=1)
   n7 = Get Control Rig Channel by Type and Name(
    execute=@n2.then_3
    Level Sequence=Get Level Sequence()
    Channel Type=&MovieSceneScriptingBoolChannel
    Channel Name=Use Legacy Brow Poses
    Log=true
   )
   > "Update Value (Bool Channel)"!(execute, self, Bool Channel:Object Reference=@n7, Value:bool=false) -> (then)
  ]}
 Fix Mouth {nodes [
   n1 = Fix Mouth!() -> (then, Level Sequence:Level Sequence Soft Object Reference, Mouth Offsets:Array of Transforms) {is=title K2Node_FunctionEntry}
   > PrintString(InString=Append(A="Fixing Level Sequence : ", B=GetSoftPtrAssetName(Asset=@n1), C=" -----------------------"))
   > n2 = Sequence!(execute) -> (then_0, then_1, then_2, then_3, then_4:Exec, then_5:Exec, then_6:Exec)
   > n3 = Sequence!(execute) -> (then_0, then_1, then_2)
   > n4 = Get Control Rig Channel by Type and Name(
    Level Sequence=Get Level Sequence()
    Channel Type=&MovieSceneScriptingFloatChannel
    Channel Name=mouth_move_ctrl.Location.X
    Log=true
   )
   > n5 = Cast To MovieSceneScriptingFloatChannel(Object=@n4)
   > Set Move U Channel(Move U Channel:Movie Scene Scripting Float Channel Object Reference=@n5) -> (Output_Get:Movie Scene Scripting Float Channel Object Reference)
   > n6 = Get Control Rig Channel by Type and Name(
    Level Sequence=Get Level Sequence()
    Channel Type=&MovieSceneScriptingFloatChannel
    Channel Name=mouth_move_ctrl.Location.Z
    Log=true
   )
   > n7 = Cast To MovieSceneScriptingFloatChannel(Object=@n6)
   > Set Move V Channel(Move V Channel:Movie Scene Scripting Float Channel Object Reference=@n7) -> (Output_Get:Movie Scene Scripting Float Channel Object Reference)
   n8 = Get Control Rig Channel by Type and Name(
    execute=@n3.then_1
    Level Sequence=Get Level Sequence()
    Channel Type=&MovieSceneScriptingFloatChannel
    Channel Name=mouth_move_ctrl.Scale.X
    Log=true
   )
   > n9 = Cast To MovieSceneScriptingFloatChannel(Object=@n8)
   > Set Scale U Channel(Scale U Channel:Movie Scene Scripting Float Channel Object Reference=@n9) -> (Output_Get:Movie Scene Scripting Float Channel Object Reference)
   > n10 = Get Control Rig Channel by Type and Name(
    Level Sequence=Get Level Sequence()
    Channel Type=&MovieSceneScriptingFloatChannel
    Channel Name=mouth_move_ctrl.Scale.Z
    Log=true
   )
   > n11 = Cast To MovieSceneScriptingFloatChannel(Object=@n10)
   > Set Scale V Channel(Scale V Channel:Movie Scene Scripting Float Channel Object Reference=@n11) -> (Output_Get:Movie Scene Scripting Float Channel Object Reference)
   n12 = Get Control Rig Channel by Type and Name(
    execute=@n3.then_2
    Level Sequence=Get Level Sequence()
    Channel Type=&MovieSceneScriptingIntegerChannel
    Channel Name=mouth_pose
    Log=true
   )
   > n13 = Cast To MovieSceneScriptingIntegerChannel!(execute, Object=@n12) -> (then, CastFailed, AsMovie Scene Scripting Integer Channel:Movie Scene Scripting Integer Channel Object Reference) {is=title K2Node_DynamicCast}
   > Set Mouth Pose Channel(Mouth Pose Channel:Movie Scene Scripting Integer Channel Object Reference=@n13) -> (Output_Get:Movie Scene Scripting Integer Channel Object Reference)
   n14 = Get Keyed Frames!(
    execute=@n2.then_1
    self
    Level Sequence:Level Sequence Soft Object Reference=Get Level Sequence()
    Channels:Array of Movie Scene Scripting Channel Object References=Make Array!(
     "[0]":Movie Scene Scripting Channel Object Reference=Get Move U Channel() -> (Move U Channel:Movie Scene Scripting Float Channel Object Reference)
     "[1]":Movie Scene Scripting Channel Object Reference=Get Move V Channel() -> (Move V Channel:Movie Scene Scripting Float Channel Object Reference)
     "[2]":Movie Scene Scripting Channel Object Reference=Get Scale U Channel() -> (Scale U Channel:Movie Scene Scripting Float Channel Object Reference)
     "[3]":Movie Scene Scripting Channel Object Reference=Get Scale V Channel() -> (Scale V Channel:Movie Scene Scripting Float Channel Object Reference)
     "[4]":Movie Scene Scripting Channel Object Reference=Get Mouth Pose Channel() -> (Mouth Pose Channel:Movie Scene Scripting Integer Channel Object Reference)
    ) -> (Array:Array of Movie Scene Scripting Channel Object References)
   ) -> (then, Frames1:Array of Integers) {is=title K2Node_CallFunction}
   > Set Keyed Frames(Keyed Frames:Array of Integers=@n14) -> (Output_Get:Array of Integers)
   n15 = GetCurrentTime!(execute=@n2.then_2) -> (then, ReturnValue:int=0)
   > Set Current Time(Current Time:int=@n15) -> (Output_Get:int=0)
   > SetCurrentTime!(execute, NewFrame:int=-1) -> (then)
   > Clear(TargetArray:"Array of Float (single-precision)s"=Get Curve MouthU Values() -> (Curve MouthU Values:"Array of Float (single-precision)s"))
   > Clear(TargetArray=Get Curve MouthV Values() -> (Curve MouthV Values:"Array of Float (single-precision)s"))
   > Clear(TargetArray=Get Curve MouthScaleU Values() -> (Curve MouthScaleU Values:"Array of Float (single-precision)s"))
   > Clear(TargetArray=Get Curve MouthScaleV Values() -> (Curve MouthScaleV Values:"Array of Float (single-precision)s"))
   > n16 = For Each Loop(Array:Array of Integers=Get Keyed Frames() -> (Keyed Frames:Array of Integers) {is=title K2Node_VariableGet}) -> (Array Element:int)
   > SetCurrentTime(NewFrame=@n16)
   > n17 = Sequence!(execute) -> (then_0, then_1, then_2, then_3:Exec)
   n18 = Get Level Sequence()
   > n19 = Get Control Rig Curve Value!(execute, self, Level Sequence:Soft Object Reference=@n18, Curve Name:Name=MouthU) -> (then, Found:bool, Value:float) {is=title K2Node_CallFunction}
   > Add(TargetArray:"Array of Float (single-precision)s"=Get Curve MouthU Values(), NewItem:"Float (single-precision) (by ref)"=@n19.Value)
   n20 = Get Control Rig Curve Value(execute=@n17.then_1, Level Sequence=@n18, Curve Name=MouthV)
   > Add(TargetArray=Get Curve MouthV Values(), NewItem=@n20.Value)
   n21 = "Get Control Rig Control Value (Transform)"!(execute=@n17.then_2, self, Level Sequence:Soft Object Reference=@n18, Control Name:Name=mouth_move_ctrl, Frame:int=@n16) -> (then, Found:bool, Location:Vector, Rotation:Rotator, Scale:Vector) {is=title K2Node_CallFunction}
   > Get Control Rig Curve Value(Level Sequence=@n18, Curve Name=MouthScaleU)
   n22 = BreakVector(InVec=@n21.Scale)
   > Add(TargetArray=Get Curve MouthScaleU Values(), NewItem=@n22)
   Get Control Rig Curve Value(execute=@n17.then_3, Level Sequence=@n18, Curve Name=MouthScaleV)
   > Add(TargetArray=Get Curve MouthScaleV Values(), NewItem=@n22.Z)
   n23 = For Each Loop(Exec=@n2.then_4, Array=Get Keyed Frames())
   > Print Integer!(
    Execute:Exec
    Int:int=@n23
    OptionalString:String="Frame ================"
    Duration:float=2
    TextColor:Linear Color Structure=(R=0,G=0.6588,B=1,A=1)
   ) -> (Then:Exec) {is=title K2Node_MacroInstance}
   > n24 = "Get Control Rig Control Value (Integer)"!(execute, self, Level Sequence:Soft Object Reference=Get Level Sequence(), Control Name:Name=mouth_pose, Frame:int=@n23) -> (then, Found:bool, Value:int=0) {is=title K2Node_CallFunction}
   > n25 = Get Pose Offset Transform!(execute, self, Array=Get Mouth Offsets() -> (Mouth Offsets:Array of Transforms), Pose=@n24.Value) -> (then, MouthU, MouthV, MouthScaleU, MouthScaleV)
   > Print Float!(
    Execute:Exec
    Float:double=@n25.MouthScaleU
    OptionalString:String="Pose Scale U "
    Duration:float=2
    TextColor:Linear Color Structure=(R=0,G=0.6588,B=1,A=1)
   ) -> (Then:Exec) {is=title K2Node_MacroInstance}
   > Print Float(Float=@n25.MouthScaleV, OptionalString="Pose Scale V ", Duration=2, TextColor=(R=0,G=0.6588,B=1,A=1))
   > n26 = Sequence()
   n27 = Make Frame Number!(Value:int=@n23) -> (FrameNumber:Frame Number Structure) {is=title K2Node_MakeStruct}
   n28 = "Get (a ref)"(Array:"Array of Float (single-precision)s"=Get Curve MouthU Values(), Dimension 1=@n23.Array Index) -> (Output:"Float (single-precision) (by ref)")
   > "Add Key (Float)"!(
    execute
    self:Movie Scene Scripting Float Channel Object Reference=Get Move U Channel()
    InTime:"Frame Number Structure (by ref)"=@n27
    NewValue:float=float / float(A=float - float(A=@n28, B=@n25), B=0.03)
    SubFrame:float=0
    TimeUnit:EMovieSceneTimeUnit Enum=DisplayRate
    InInterpolation:EMovieSceneKeyInterpolation Enum=Constant
   ) -> (then, ReturnValue:Movie Scene Scripting Float Key Object Reference)
   n29 = "Get (a ref)"(Array=Get Curve MouthV Values(), Dimension 1=@n23.Array Index)
   "Add Key (Float)"(
    execute=@n26.then_1
    self=Get Move V Channel()
    InTime=@n27
    NewValue=float / float(A=float - float(A=@n29, B=@n25.MouthV), B=0.048)
    SubFrame=0
    TimeUnit=DisplayRate
    InInterpolation=Constant
   )
   n30 = "Get (a ref)"(Array=Get Curve MouthScaleU Values(), Dimension 1=@n23.Array Index)
   Print Float(Execute=@n26.then_2, Float=@n30, OptionalString="Curve Scale U ", Duration=2, TextColor=(R=0,G=0.6588,B=1,A=1))
   n31 = float * float(A=float / float(A=@n30, B=@n25.MouthScaleU), B=1.25)
   > Print Float(Float=@n31, OptionalString="Result Scale U ", Duration=2, TextColor=(R=0,G=0.6588,B=1,A=1))
   > "Add Key (Float)"(self=Get Scale U Channel(), InTime=@n27, NewValue=@n31, SubFrame=0, TimeUnit=DisplayRate, InInterpolation=Constant)
   n32 = "Get (a ref)"(Array=Get Curve MouthScaleV Values(), Dimension 1=@n23.Array Index)
   Print Float(Execute=@n26.then_3, Float=@n32, OptionalString="Curve Scale V ", Duration=2, TextColor=(R=0,G=0.6588,B=1,A=1))
   n33 = float * float(A=float / float(A=@n32, B=@n25.MouthScaleV), B=1.25)
   > Print Float(Float=@n33, OptionalString="Result Scale V ", Duration=2, TextColor=(R=0,G=0.6588,B=1,A=1))
   > "Add Key (Float)"(self=Get Scale V Channel(), InTime=@n27, NewValue=@n33, SubFrame=0, TimeUnit=DisplayRate, InInterpolation=Constant)
   SetCurrentTime(execute=@n2.then_5, NewFrame=Get Current Time() -> (Current Time:int=0))
   n34 = Get Control Rig Channel by Type and Name(
    execute=@n2.then_6
    Level Sequence=Get Level Sequence()
    Channel Type=&MovieSceneScriptingBoolChannel
    Channel Name=Use Legacy Face Transforms
    Log=true
   )
   > "Update Value (Bool Channel)"(Bool Channel=@n34, Value=false)
  ]}
 Get Control Rig Channel by Type and Name {nodes [
   n1 = Get Control Rig Channel by Type and Name!() -> (
    then
    Level Sequence:Level Sequence Soft Object Reference
    Channel Type:Movie Scene Scripting Channel Class Reference
    Channel Name:Name
    Log:bool=true
   ) {is=title K2Node_FunctionEntry}
   > n2 = LoadAsset_Blocking(Asset=@n1)
   > n3 = Cast To LevelSequence(Object=@n2)
   > n4 = GetControlRigs!(execute, LevelSequence:Level Sequence Object Reference=@n3) -> (then, ReturnValue:Array of Control Rig Sequencer Binding Proxy Structures)
   > n5 = For Each Loop(Array:Array of Control Rig Sequencer Binding Proxy Structures=@n4) -> (Array Element:Control Rig Sequencer Binding Proxy Structure)
   Return Node!(execute=@n5.Completed, Array Element:Movie Scene Scripting Channel Object Reference)
   n6 = Break Control Rig Sequencer Binding Proxy!(ControlRigSequencerBindingProxy:"Control Rig Sequencer Binding Proxy Structure (by ref)"=@n5) -> (
    Proxy:Movie Scene Binding Proxy Structure
    ControlRig:Control Rig Object Reference
    Track:Movie Scene Control Rig Parameter Track Object Reference
   ) {is=title K2Node_BreakStruct}
   > n7 = GetTracks!(execute, InBinding:"Movie Scene Binding Proxy Structure (by ref)"=@n6) -> (then, ReturnValue:Array of Movie Scene Track Object References)
   > n8 = For Each Loop(Array:Array of Movie Scene Track Object References=@n7) -> (Array Element:Movie Scene Track Object Reference)
   > n9 = GetSections!(execute, Track:Movie Scene Track Object Reference=@n8) -> (then, ReturnValue:Array of Movie Scene Section Object References)
   > n10 = For Each Loop(Array:Array of Movie Scene Section Object References=@n9) -> (Array Element:Movie Scene Section Object Reference)
   > n11 = GetChannelsByType!(
    execute
    Section:Movie Scene Section Object Reference=@n10
    ChannelType:Movie Scene Scripting Channel Class Reference=Get Channel Type() -> (Channel Type:Movie Scene Scripting Channel Class Reference)
   ) -> (then, ReturnValue:Array of Movie Scene Scripting Channel Object References)
   > n12 = For Each Loop with Break!(Exec:Exec, Array:Array of Movie Scene Scripting Channel Object References=@n11, Break:Exec) -> (LoopBody:Exec, Array Element:Movie Scene Scripting Channel Object Reference, Array Index:int, Completed:Exec)
   > Branch(Condition="Equal (Name)"(A=Get ChannelName(self=@n12) -> (ChannelName=None), B=Get Channel Name() -> (Channel Name:Name=None)))
   n13 = Get Channel Name() -> (Channel Name=None)
   > PrintString(
    InString=Append!(A="Found Channel (", B="To String (Name)"(InName=@n13), C=") on Rig : ", D:String=GetDisplayName(Object=@n6.ControlRig), E:String) -> (ReturnValue)
    bPrintToScreen=@n1.Log
    bPrintToLog=@n1.Log
   )
   > Return Node(Array Element=@n12)
  ]}
 "Get Control Rig Control Value (Integer)" {nodes [
   n1 = "Get Control Rig Control Value (Integer)"!() -> (then, Level Sequence:Soft Object Reference, Control Name:Name, Frame:int) {is=title K2Node_FunctionEntry}
   > n2 = LoadAsset_Blocking(Asset=@n1)
   > n3 = Cast To LevelSequence(Object=@n2)
   > n4 = GetControlRigs(LevelSequence=@n3)
   > n5 = GetLocalControlRigInt!(
    execute
    LevelSequence:Level Sequence Object Reference=@n3
    ControlRig:Control Rig Object Reference=Break Control Rig Sequencer Binding Proxy(ControlRigSequencerBindingProxy="Get (a ref)"(Array:Array of Control Rig Sequencer Binding Proxy Structures=@n4, Dimension 1=0) -> (Output:"Control Rig Sequencer Binding Proxy Structure (by ref)")).ControlRig
    ControlName:Name=@n1.Control Name
    Frame:Frame Number Structure=Make Frame Number(Value=@n1.Frame)
    TimeUnit:EMovieSceneTimeUnit Enum=DisplayRate
   ) -> (then, ReturnValue:int=0)
   > Return Node!(execute, Found:bool=true, Value:int=@n5)
  ]}
 "Get Control Rig Control Value (Transform)" {nodes [
   n1 = "Get Control Rig Control Value (Transform)"!() -> (then, Level Sequence:Soft Object Reference, Control Name:Name, Frame:int) {is=title K2Node_FunctionEntry}
   > n2 = LoadAsset_Blocking(Asset=@n1)
   > n3 = Cast To LevelSequence(Object=@n2)
   > n4 = GetControlRigs(LevelSequence=@n3)
   > n5 = GetLocalControlRigEulerTransform!(
    execute
    LevelSequence:Level Sequence Object Reference=@n3
    ControlRig:Control Rig Object Reference=Break Control Rig Sequencer Binding Proxy(ControlRigSequencerBindingProxy="Get (a ref)"(Array=@n4, Dimension 1=0)).ControlRig
    ControlName:Name=@n1.Control Name
    Frame:Frame Number Structure=Make Frame Number(Value=@n1.Frame)
    TimeUnit:EMovieSceneTimeUnit Enum=DisplayRate
   ) -> (then, ReturnValue:Euler Transform Structure)
   n6 = Break Euler Transform!(EulerTransform:"Euler Transform Structure (by ref)"=@n5) -> (Location:Vector=~, Rotation:Rotator=~, Scale:Vector=~) {is=title K2Node_BreakStruct}
   > Return Node!(execute, Found=true, Location:Vector=@n6, Rotation:Rotator=@n6.Rotation, Scale:Vector=@n6.Scale)
  ]}
 Get Control Rig Curve Value {nodes [
   n1 = Get Control Rig Curve Value!() -> (then, Level Sequence:Soft Object Reference, Curve Name:Name) {is=title K2Node_FunctionEntry}
   > n2 = LoadAsset_Blocking(Asset=@n1)
   > n3 = Cast To LevelSequence(Object=@n2)
   > n4 = GetControlRigs(LevelSequence=@n3)
   > n5 = For Each Loop(Array:Array of Control Rig Sequencer Binding Proxy Structures=@n4) -> (Array Element:Control Rig Sequencer Binding Proxy Structure)
   Return Node!(execute=@n5.Completed, Found=false, Value:float=0)
   n6 = GetHierarchy!(self:Control Rig Object Reference=Break Control Rig Sequencer Binding Proxy(ControlRigSequencerBindingProxy=@n5).ControlRig) -> (ReturnValue:Rig Hierarchy Object Reference)
   > n7 = For Each Loop with Break(Array:Array of Rig Element Key Structures=Get Curves!(self:Rig Hierarchy Object Reference=@n6) -> (ReturnValue:Array of Rig Element Key Structures) {is=title K2Node_CallFunction}) -> (Array Element:Rig Element Key Structure)
   > Branch(Condition="Equal (Name)"(
     A=Break Rig Element Key!(RigElementKey:"Rig Element Key Structure (by ref)"=@n7) -> (Type:ERigElementType Enum, Name:Name) {is=title K2Node_BreakStruct}.Name
     B=@n1.Curve Name
    ))
   n8 = Make Rig Element Key!(Type:ERigElementType Enum=Curve, Name:Name=@n1.Curve Name) -> (RigElementKey:Rig Element Key Structure) {is=title K2Node_MakeStruct}
   > Return Node(Found=true, Value=GetCurveValue!(self:Rig Hierarchy Object Reference=@n6, InKey:Rig Element Key Structure=@n8) -> (ReturnValue))
   GetCurveValue(self=@n6, InKey=@n8)
  ]}
 "Get KeyValue Array (Bool Channel)" {nodes [
   n1 = "Get KeyValue Array (Bool Channel)"!() -> (then, Level Sequence:Level Sequence Soft Object Reference, Channel Name:Name) {is=title K2Node_FunctionEntry}
   > Clear!(execute, TargetMap:Map of Integers to Integers=Get Key Map() -> (Key Map:Map of Integers to Integers)) -> (then) {is=title K2Node_CallFunction}
   > n2 = Get Control Rig Channel by Type and Name(Level Sequence=@n1, Channel Type=&MovieSceneScriptingIntegerChannel, Channel Name=@n1.Channel Name, Log=true) {is=title K2Node_CallFunction}
   > n3 = Cast To MovieSceneScriptingIntegerChannel(Object=@n2)
   > n4 = For Each Loop(Array:Array of Movie Scene Scripting Key Object References="Get Keys (Integer)"!(self:Movie Scene Scripting Integer Channel Object Reference=@n3) -> (ReturnValue:Array of Movie Scene Scripting Key Object References) {is=title K2Node_CallFunction}) -> (Array Element:Movie Scene Scripting Key Object Reference)
   Return Node!(execute=@n4.Completed, Key Map1:Map of Integers to Integers=Get Key Map())
   > n5 = Cast To MovieSceneScriptingIntegerKey!(execute, Object=@n4) -> (then, CastFailed, AsMovie Scene Scripting Integer Key:Movie Scene Scripting Integer Key Object Reference) {is=title K2Node_DynamicCast}
   n6 = "Get Time (Integer)"!(self:Movie Scene Scripting Integer Key Object Reference=@n5, TimeUnit:EMovieSceneTimeUnit Enum=DisplayRate) -> (ReturnValue:Frame Time Structure) {is=title K2Node_CallFunction}
   > Add!(
    execute
    TargetMap:Map of Integers to Integers=Get Key Map()
    Key:"Integer (by ref)"=Break Frame Number(FrameNumber=Break Frame Time(FrameTime=@n6) -> (SubFrame=~))
    Value:"Integer (by ref)"="Get Value (Integer)"!(self:Movie Scene Scripting Integer Key Object Reference=@n5) -> (ReturnValue:int=0) {is=title K2Node_CallFunction}
   ) -> (then) {is=title K2Node_CallFunction}
  ]}
 Get Keyed Frames {nodes [
   n1 = Get Keyed Frames!() -> (then, Level Sequence:Level Sequence Soft Object Reference, Channels:Array of Movie Scene Scripting Channel Object References) {is=title K2Node_FunctionEntry}
   > Clear(TargetArray:Array of Integers=Get Frames() -> (Frames:Array of Integers)) {is=title K2Node_CallArrayFunction}
   > n2 = Sequence!(execute) -> (then_0, then_1, then_2)
   > n3 = For Each Loop(Array:Array of Movie Scene Scripting Channel Object References=@n1.Channels) -> (Array Element:Movie Scene Scripting Channel Object Reference)
   > n4 = Cast To MovieSceneScriptingIntegerChannel(Object=@n3)
   > n5 = For Each Loop(Array:Array of Movie Scene Scripting Key Object References="Get Keys (Integer)"(self=@n4)) -> (Array Element:Movie Scene Scripting Key Object Reference)
   > n6 = Cast To MovieSceneScriptingIntegerKey(Object=@n5)
   n7 = "Get Time (Integer)"(self=@n6, TimeUnit=DisplayRate)
   > Add Unique(
    TargetArray:Array of Integers=Get Frames()
    NewItem:"Integer (by ref)"=Break Frame Number(FrameNumber=Break Frame Time(FrameTime=@n7) -> (SubFrame=~))
   )
   n8 = Cast To MovieSceneScriptingFloatChannel(execute=@n4.CastFailed, Object=@n3)
   > n9 = For Each Loop(Array="Get Keys (Float)"(self=@n8))
   > n10 = Cast To MovieSceneScriptingFloatKey(Object=@n9)
   n11 = "Get Time (Float)"(self=@n10, TimeUnit=DisplayRate)
   > Add Unique(TargetArray=Get Frames(), NewItem=Break Frame Number(FrameNumber=Break Frame Time(FrameTime=@n11) -> (SubFrame=~)))
   Clear(execute=@n2.then_1, TargetArray=Get Sorted Frames() -> (Sorted Frames:Array of Integers))
   > For Loop(LastIndex=Length(TargetArray:Array of Integers=Get Frames()))
   n12 = Get Frames()
   n13 = MinOfIntArray!(IntArray:Array of Integers=@n12) -> (IndexOfMinValue:int=0, MinValue:int=0)
   > Add(TargetArray:Array of Integers=Get Sorted Frames(), NewItem:"Integer (by ref)"=@n13.MinValue) {is=title K2Node_CallArrayFunction}
   > Remove Index!(execute, TargetArray:Array of Integers=@n12, IndexToRemove:int=@n13) -> (then)
   Return Node!(execute=@n2.then_2, Frames1:Array of Integers=Get Sorted Frames())
   Get Frames()
  ]}
 Get Pose Offset Transform {nodes [
   n1 = Get Pose Offset Transform!() -> (then, Array:Array of Transforms, Pose:int) {is=title K2Node_FunctionEntry}
   n2 = BreakTransform(InTransform="Get (a ref)"(Array:Array of Transforms=@n1, Dimension 1=@n1.Pose) -> (Output:"Transform (by ref)"))
   n3 = BreakVector(InVec=@n2.Scale)
   n4 = BreakVector(InVec=@n2)
   > Return Node!(execute, MouthU:double=@n4, MouthV:double=@n4.Z, MouthScaleU:double=@n3, MouthScaleV:double=@n3.Z)
  ]}
 Get Pose at Frame {nodes [
   n1 = Get Pose at Frame!() -> (then, TargetMap:Map of Integers to Integers, Time:int) {is=title K2Node_FunctionEntry}
   > Set Delta(Delta:int=9999999) -> (Output_Get:int=0)
   > n2 = Keys!(execute, TargetMap:Map of Integers to Integers=@n1) -> (then, Keys:Array of Integers)
   > n3 = For Each Loop(Array:Array of Integers=@n2) -> (Array Element:int)
   Return Node!(execute=@n3.Completed, Pose:int=Get Value() -> (Value:int=0))
   n4 = Get Time() -> (Time:int=0)
   > Branch(Condition="integer <= integer"(A=@n3, B=@n4))
   n5 = int - int(A=@n4, B=@n3.Array Index)
   > Branch(Condition="integer < integer"(A=@n5, B=Get Delta() -> (Delta:int=0)))
   > Set Delta(Delta=@n5) -> (Output_Get=0)
   > n6 = Values!(execute, TargetMap:Map of Integers to Integers=Get TargetMap() -> (TargetMap:Map of Integers to Integers)) -> (then, Values:Array of Integers)
   > Set Value(Value:int="Get (a ref)"(Array:Array of Integers=@n6, Dimension 1=@n3.Array Index) -> (Output:"Integer (by ref)")) -> (Output_Get:int=0)
   Length(TargetArray=@n2)
  ]}
 Get Transform Channels {nodes [
   n1 = Get Transform Channels!() -> (then, Level Sequence:Level Sequence Soft Object Reference) {is=title K2Node_FunctionEntry}
   > Clear(TargetArray:Array of Movie Scene Scripting Channel Object References=Get Location Channels() -> (Location Channels:Array of Movie Scene Scripting Channel Object References))
   > Clear(TargetArray=Get Scale Channels() -> (Scale Channels:Array of Movie Scene Scripting Channel Object References))
   > n2 = For Each Loop(Array:Array of Strings=Make Array!("[0]":String=X, "[1]":String=Z) -> (Array:Array of Strings)) -> (Array Element:String)
   Return Node!(
    execute=@n2.Completed
    Location Channels1:Array of Movie Scene Scripting Channel Object References=Get Location Channels()
    Scale Channels1:Array of Movie Scene Scripting Channel Object References=Get Scale Channels()
   )
   > n3 = Get Control Rig Channel by Type and Name(
    Level Sequence=@n1
    Channel Type=&MovieSceneScriptingFloatChannel
    Channel Name=String To Name(InString=Append!(A=mouth_move_ctrl.Location., B=@n2) -> (ReturnValue))
    Log=true
   )
   > Add(
    TargetArray:Array of Movie Scene Scripting Channel Object References=Get Location Channels()
    NewItem:"Movie Scene Scripting Channel Object Reference (by ref)"=@n3
   )
   > n4 = Get Control Rig Channel by Type and Name(
    Level Sequence=@n1
    Channel Type=&MovieSceneScriptingFloatChannel
    Channel Name=String To Name(InString=Append(A=mouth_move_ctrl.Scale., B=@n2))
    Log=true
   )
   > Add(TargetArray=Get Scale Channels(), NewItem=@n4)
  ]}
 "Has Constant Value (Bool Channel )" {nodes [
   n1 = "Has Constant Value (Bool Channel )"!() -> (then, Channel:Movie Scene Scripting Channel Object Reference) {is=title K2Node_FunctionEntry}
   > n2 = Cast To MovieSceneScriptingBoolChannel!(execute, Object=@n1) -> (then, CastFailed, AsMovie Scene Scripting Bool Channel:Movie Scene Scripting Bool Channel Object Reference) {is=title K2Node_DynamicCast}
   > n3 = For Each Loop(Array:Array of Movie Scene Scripting Key Object References="Get Keys (Bool)"!(self:Movie Scene Scripting Bool Channel Object Reference=@n2) -> (ReturnValue:Array of Movie Scene Scripting Key Object References) {is=title K2Node_CallFunction}) -> (Array Element:Movie Scene Scripting Key Object Reference)
   Return Node!(execute=@n3.Completed, Constant:bool=true)
   > n4 = Cast To MovieSceneScriptingBoolKey!(execute, Object=@n3) -> (then, CastFailed, AsMovie Scene Scripting Bool Key:Movie Scene Scripting Bool Key Object Reference) {is=title K2Node_DynamicCast}
   > n5 = Branch(Condition="Get Value (Bool)"!(self:Movie Scene Scripting Bool Key Object Reference=@n4) -> (ReturnValue:bool) {is=title K2Node_CallFunction})
   Return Node(execute=@n5.else, Constant=false)
  ]}
 "Has Keys (Integer Channel)" {nodes [
   n1 = "Has Keys (Integer Channel)"!() -> (then, Object:Object Reference) {is=title K2Node_FunctionEntry}
   > n2 = Cast To MovieSceneScriptingIntegerChannel(Object=@n1)
   n3 = "Get Keys (Integer)"(self=@n2)
   > Return Node!(execute, ReturnValue:bool="integer > integer"(A=Length(TargetArray:Array of Movie Scene Scripting Key Object References=@n3)))
  ]}
 "Match Value (Bool Channel)" {nodes [
   n1 = "Match Value (Bool Channel)"!() -> (then, NewParam:Movie Scene Scripting Channel Object Reference, Value:bool=~) {is=title K2Node_FunctionEntry}
   > n2 = Cast To MovieSceneScriptingBoolChannel(Object=@n1)
   n3 = "Get Keys (Bool)"(self=@n2)
   > n4 = Branch(Condition="Equal (Integer)"(A=Length(TargetArray=@n3)))
   n5 = "Get Default (Bool)"!(self:Movie Scene Scripting Bool Channel Object Reference=@n2) -> (ReturnValue:bool) {is=title K2Node_CallFunction}
   > Branch(Condition="Equal (Boolean)"(A:bool=@n5, B:bool=@n1.Value) -> (ReturnValue:bool=~) {is=title K2Node_PromotableOperator})
   > Print Bool!(
    Execute:Exec
    Bool:bool=@n5
    OptionalString:String="No Key Found, Default Match"
    Duration:float=2
    TextColor:Linear Color Structure=(R=0,G=0.6588,B=1,A=1)
   ) -> (Then:Exec) {is=title K2Node_MacroInstance}
   > Return Node!(execute, Result:bool=true)
   n6 = For Each Loop with Break(Exec=@n4.else, Array:Array of Movie Scene Scripting Key Object References=@n3) -> (Array Element:Movie Scene Scripting Key Object Reference)
   PrintString(execute=@n6.Completed, InString=Key not Found)
   > Return Node(Result=false)
   n7 = Cast To MovieSceneScriptingBoolKey(execute=@n6.LoopBody, Object=@n6)
   > Branch(Condition="Equal (Boolean)"(A="Get Value (Bool)"(self=@n7), B=@n1.Value) -> (ReturnValue=~))
   n8 = FrameNumber to Integer!(InFrameNumber:"Frame Number Structure (by ref)"="Get Time (Bool)"!(self:Movie Scene Scripting Bool Key Object Reference=@n7, TimeUnit:EMovieSceneTimeUnit Enum=DisplayRate) -> (ReturnValue_FrameNumber:Frame Number Structure, ReturnValue_SubFrame:float=~) {is=title K2Node_CallFunction}) -> (ReturnValue:int=0)
   > PrintString(InString=Append!(A="Found Key at Frame : ", B="To String (Integer)"(InInt=@n8), C:String) -> (ReturnValue))
   > Return Node(Result=true)
  ]}
 "Remap Value (Integer Channel)" {nodes [
   n1 = "Remap Value (Integer Channel)"!() -> (then, NewParam:Movie Scene Scripting Channel Object Reference, Brow Pose Map:Array of Integers) {is=title K2Node_FunctionEntry}
   > n2 = Cast To MovieSceneScriptingIntegerChannel(Object=@n1)
   > n3 = For Each Loop(Array="Get Keys (Integer)"(self=@n2))
   Return Node!(execute=@n3.Completed)
   > n4 = Cast To MovieSceneScriptingIntegerKey(Object=@n3)
   > "Set Value (Integer)"!(
    execute
    self:Movie Scene Scripting Integer Key Object Reference=@n4
    InNewValue:int="Get (a ref)"(Array=@n1.Brow Pose Map, Dimension 1="Get Value (Integer)"(self=@n4) -> (ReturnValue=0))
   ) -> (then) {is=title K2Node_CallFunction}
   n5 = FrameNumber to Integer(InFrameNumber="Get Time (Integer)"!(self=@n4, TimeUnit=DisplayRate) -> (ReturnValue_FrameNumber:Frame Number Structure, ReturnValue_SubFrame:float=~)) -> (ReturnValue=0)
   > PrintString(InString=Append(A="Remap Key at Frame : ", B="To String (Integer)"(InInt=@n5)))
  ]}
 "Update Value (Bool Channel)" {nodes [
   n1 = "Update Value (Bool Channel)"!() -> (then, Bool Channel:Object Reference, Value:bool=~) {is=title K2Node_FunctionEntry}
   > n2 = Cast To MovieSceneScriptingBoolChannel(Object=@n1)
   > "Set Default (Bool)"!(execute, self:Movie Scene Scripting Bool Channel Object Reference=@n2, InDefaultValue:bool=@n1.Value) -> (then) {is=title K2Node_CallFunction}
   > n3 = For Each Loop(Array="Get Keys (Bool)"(self=@n2))
   > n4 = Cast To MovieSceneScriptingBoolKey(Object=@n3)
   > "Set Value (Bool)"!(execute, self:Movie Scene Scripting Bool Key Object Reference=@n4, InNewValue:bool=@n1.Value) -> (then) {is=title K2Node_CallFunction}
  ]}
 "Update Value (Float Channel)" {nodes [
   n1 = "Update Value (Float Channel)"!() -> (then, Float Channel:Movie Scene Scripting Channel Object Reference, Value:double=~) {is=title K2Node_FunctionEntry}
   > n2 = Cast To MovieSceneScriptingFloatChannel(Object=@n1)
   > "Set Default (Float)"!(execute, self:Movie Scene Scripting Float Channel Object Reference=@n2, InDefaultValue:float=Get Value() -> (Value:double)) -> (then) {is=title K2Node_CallFunction}
   > n3 = For Each Loop with Break(Array="Get Keys (Float)"(self=@n2))
   > n4 = Cast To MovieSceneScriptingFloatKey(Object=@n3)
   > "Set Value (Float)"(self=@n4, InNewValue=Get Value())
   Return Node!(execute, Result:bool=true)
  ]}
]
variables [
 Any {type=boolean, default=true}
 Array Element {type=MovieSceneScriptingChannel, default=null}
 Bool Channels {type="map<boolean>", default {use Legacy Brow Poses=true}}
 Brow Pose Map {type="array<integer>", default [0,4,6,1,2,8,7,9,10,8,7,11,8,9,3,4,9,8,8,8,10,10,10,11,11]}
 Can't Fix Sequences {type="array<LevelSequence>", default []}
 Constant Value {type=boolean}
 Current Location Channels {type="array<MovieSceneScriptingChannel>", default []}
 Current Scale Channels {type="array<MovieSceneScriptingChannel>", default []}
 Current Sequence {type=LevelSequence, default=null}
 Execute Brows Fix {type=boolean, default=true}
 Execute Mouth Fix {type=boolean, default=true}
 Float Channels {type="map<STRUCT_FloatMinMax>", default {}}
 Found {type="array<boolean>", default []}
 Integer Channels {type="map<STRUCT_IntegerMinMax>", default {}}
 Legacy Brow Poses Sequences {type="array<LevelSequence>", default []}
 Legacy Face Transforms Sequences {type="array<LevelSequence>", default []}
 Mouth Pose Key Map {default {}}
 Mouth Pose Offset {
  type="array<Transform>"
  default [
   {location=<0 0 -0.2016>, rotation=<rot 0 0 0>, scale=<0.9639 1 1.0667>}
   {location=<-0.0119 0 -0.1903>, rotation=<rot 0 0 0>, scale=<0.8889 1 0.8602>}
   {location=<0 0 -0.1777>, rotation=<rot 0 0 0>, scale=<1.0127 1 0.9195>}
   {location=<0 0 -0.1845>, rotation=<rot 0 0 0>, scale=<0.8889 1 0.8081>}
   {location=<0 0 -0.2041>, rotation=<rot 0 0 0>, scale=<1.0667 1 1.1765>}
   {location=<0 0 -0.1861>, rotation=<rot 0 0 0>, scale=<0.97 1 1.1268>}
   {location=<0 0 -0.1989>, rotation=<rot 0 0 0>, scale=<1.1111 1 1.0882>}
   {location=<0 0 -0.1965>, rotation=<rot 0 0 0>, scale=<1 1 1.0526>}
   {location=<0 0 -0.1721>, rotation=<rot 0 0 0>, scale=<1.0667 1 1.0959>}
   {location=<0 0 -0.1856>, rotation=<rot 0 0 0>, scale=<1.1111 1 1.1515>}
   {location=<0 0 -0.1791>, rotation=<rot 0 0 0>, scale=<1.1268 1 0.8791>}
   {location=<0 0 -0.1922>, rotation=<rot 0 0 0>, scale=<1.1394 1 1.1159>}
   {location=<0.0032 0 -0.1976>, rotation=<rot 0 0 0>, scale=<1.1268 1 1.0959>}
   {location=<0 0 -0.1877>, rotation=<rot 0 0 0>, scale=<1.02 1 0.8>}
   {location=<0 0 -0.186>, rotation=<rot 0 0 0>, scale=<0.8 1 0.72>}
   {location=<0 0 -0.1977>, rotation=<rot 0 0 0>, scale=<0.9756 1 1.2308>}
   {location=<-0.0047 0 -0.215>, rotation=<rot 0 0 0>, scale=<1.05 1 1>}
   {location=<0 0 -0.2013>, rotation=<rot 0 0 0>, scale=<1.1765 1 0.8>}
   {location=<0 0 -0.1864>, rotation=<rot 0 0 0>, scale=<1.039 1 0.898>}
   {location=<0.0016 0 -0.2067>, rotation=<rot 0 0 0>, scale=<0.81 1 0.88>}
   {location=<0 0 -0.2011>, rotation=<rot 0 0 0>, scale=<0.8 1 0.97>}
   {location=<0 0 -0.2104>, rotation=<rot 0 0 0>, scale=<0.8511 1 0.9756>}
   {location=<-0.0192 0 -0.2097>, rotation=<rot 0.0092 0 0>, scale=<1.0959 1 1.0959>}
   {location=<0.0016 0 -0.1949>, rotation=<rot 0 0 0>, scale=<1.0667 1 1.0811>}
   {location=<0 0 -0.1901>, rotation=<rot 0 0 0>, scale=<1.0127 1 1.1111>}
   {location=<0 0 -0.2015>, rotation=<rot 0 0 0>, scale=<1 1 1.15>}
   {location=<0.0032 0 -0.1782>, rotation=<rot 0 0 0>, scale=<0.8711 1 0.8>}
   {location=<0 0 -0.1822>, rotation=<rot 0 0 0>, scale=<1.0811 1 0.9524>}
   {location=<0 0 -0.1892>, rotation=<rot 0 0 0>, scale=<1.0811 1 1.3>}
   {location=<0 0 -0.1733>, rotation=<rot 0 0 0>, scale=<1.0959 1 1.3115>}
   {location=<-0.0016 0 -0.175>, rotation=<rot 0 0 0>, scale=<0.9524 1 1.0779>}
   {location=<-0.0038 0 -0.1991>, rotation=<rot 0 0 0>, scale=<1.1111 1 1.4035>}
   {location=<-0.0063 0 -0.1917>, rotation=<rot 0 0 0>, scale=<0.81 1 0.8>}
   {location=<-0.0016 0 -0.1932>, rotation=<rot 0 0 0>, scale=<1.0127 1 1.0886>}
   {location=<0.0032 0 -0.1854>, rotation=<rot -0.0157 0 0>, scale=<1.0353 1 1.1047>}
   {location=<0 0 -0.1781>, rotation=<rot 0 0 0>, scale=<0.9756 1 0.9655>}
   {location=<-0.003 0 -0.184>, rotation=<rot 0 0 0>, scale=<0.9877 1 0.95>}
   {location=<0 0 -0.2029>, rotation=<rot 0 0 0>, scale=<0.9091 1 1.0959>}
   {location=<0 0 -0.21>, rotation=<rot 0 0 0>, scale=<1.0811 1 0.9756>}
   {location=<0 0 -0.206>, rotation=<rot 0 0 0>, scale=<0.9524 1 0.8>}
   {location=<0 0 -0.1904>, rotation=<rot 0 0 0>, scale=<1.0127 1 0.8>}
   {location=<0 0 -0.2033>, rotation=<rot 0 0 0>, scale=<0.8 1 0.8>}
   {location=<0 0 -0.2>, rotation=<rot 0 0 0>, scale=<0.8 1 0.8>}
   {location=<0 0 -0.2>, rotation=<rot 0 0 0>, scale=<0.8 1 0.8>}
   {location=<0 0 -0.2>, rotation=<rot 0 0 0>, scale=<0.8 1 0.8>}
   {location=<0 0 -0.2>, rotation=<rot 0 0 0>, scale=<0.8 1 0.8>}
  ]
 }
 Run Fix Brows {type=boolean}
 Run Fix Mouth {type=boolean}
 SaveGame {type=SG_AnimationChannelInspector_C, default=null}
 Sequences {type="array<LevelSequence>", default []}
]
widgets {
 info {rootWidgetClass=&VerticalBox, widgetCount=14}
 widgets [
  VerticalBox_0 {widgetClassPath=&VerticalBox, widget=VerticalBox_0, properties {clipping=ClipToBounds}}
  GetFromContentBrowser_btn {
   widgetClassPath=&EditorUtilityButton
   parent=VerticalBox_0
   widget=GetFromContentBrowser_btn
   properties {visibility=Hidden}
   slot {class=VerticalBoxSlot}
  }
  TextBlock_1 {
   widgetClassPath=&TextBlock
   parent=GetFromContentBrowser_btn
   widget=TextBlock_1
   properties {text=Get from Content Browser}
   slot {class=ButtonSlot}
  }
  EditorUtilityScrollBox_215 {
   widgetClassPath=&EditorUtilityScrollBox
   parent=VerticalBox_0
   widget=EditorUtilityScrollBox_215
   properties {alwaysShowScrollbar=true, visibility=Hidden}
   slot {class=VerticalBoxSlot, properties {size {sizeRule=Fill, value=1}}}
  }
  Input_dv {
   widgetClassPath=&DetailsView
   parent=EditorUtilityScrollBox_215
   widget=Input_dv
   properties {categoriesToShow [InputSequences]}
   slot {class=ScrollBoxSlot}
  }
  Search_btn {
   widgetClassPath=&EditorUtilityButton
   parent=VerticalBox_0
   widget=Search_btn
   properties {visibility=Hidden}
   slot {class=VerticalBoxSlot}
  }
  TextBlock_0 {widgetClassPath=&TextBlock, parent=Search_btn, widget=TextBlock_0, properties {text=Search}, slot {class=ButtonSlot}}
  EditorUtilityScrollBox_0 {
   widgetClassPath=&EditorUtilityScrollBox
   parent=VerticalBox_0
   widget=EditorUtilityScrollBox_0
   properties {alwaysShowScrollbar=true, visibility=Hidden}
   slot {class=VerticalBoxSlot, properties {size {sizeRule=Fill, value=1}}}
  }
  Output_dv {
   widgetClassPath=&DetailsView
   parent=EditorUtilityScrollBox_0
   widget=Output_dv
   properties {categoriesToShow [OutputSequences]}
   slot {class=ScrollBoxSlot}
  }
  SelectInContentBrowser_btn {
   widgetClassPath=&EditorUtilityButton
   parent=VerticalBox_0
   widget=SelectInContentBrowser_btn
   properties {visibility=Hidden}
   slot {class=VerticalBoxSlot}
  }
  TextBlock_2 {
   widgetClassPath=&TextBlock
   parent=SelectInContentBrowser_btn
   widget=TextBlock_2
   properties {text=Select in Content Browser}
   slot {class=ButtonSlot}
  }
  Options_dv {
   widgetClassPath=&DetailsView
   parent=VerticalBox_0
   widget=Options_dv
   properties {visibility=Hidden, categoriesToShow [Options]}
   slot {class=VerticalBoxSlot}
  }
  Fix_btn {widgetClassPath=&EditorUtilityButton, parent=VerticalBox_0, widget=Fix_btn, slot {class=VerticalBoxSlot}}
  TextBlock {widgetClassPath=&TextBlock, parent=Fix_btn, widget=TextBlock, properties {text=Fix Animation}, slot {class=ButtonSlot}}
 ]
}
