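; BPI_AnimationPickerWidget is a Blueprint interface for widgets used by the animation picker editor tool.
; It declares getters for all selectable elements and the primary panel, plus hooks for mouse enter and updating from the scene.
; A widget-loaded callback passes in the Control Rig sequencer binding proxy, the animatable creature and the owning Animation Picker editor utility widget.
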(fn GetallSelectableElements
  (return 0))

(fn GetPrimaryPanel
  (return 0))

(fn MouseEnter
  (params (Mouse%20Event :type Pointer%20Event%20Structure)))

(fn UpdateFromScene)

(fn Widgethasloaded
  (params (Control%20Rig%20Proxy%20Binding :type Control%20Rig%20Sequencer%20Binding%20Proxy%20Structure) (Animatable%20Creature :type Animatable%20Creature%20Object%20Reference) (Anim%20Picker%20EUW :type Animation%20Picker%20Object%20Reference)))
