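# Base Editor Utility Widget for animation pickers that select, clear and mirror Control Rig controls in Sequencer.
# Helpers select controls by button name or "_ctrl" suffix and gather picker buttons from a panel.
# Mirror Control caches selected values against counterparts found by swapping _l_/_r_, with parent lookup for animation channels.
# It then keys them at the current frame by control type, flipping transforms using per-control plane and behavioral settings, fixed defaults here.
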EditorUtilityWidgetBlueprint /AnimInEdTools/AnimationPicker/EUW_Animation_Picker_Base
graphs [
 Clear Selection {nodes [
   Clear Selection() {is=title K2Node_FunctionEntry}
   > ClearControlSelection!(
    execute
    self:Control Rig Object Reference=Get!() -> (
     Control Rig Binding_Proxy:Movie Scene Binding Proxy Structure
     Control Rig Binding_ControlRig:Control Rig Object Reference
     Control Rig Binding_Track:Movie Scene Control Rig Parameter Track Object Reference
    ).Control Rig Binding_ControlRig
    bSetupUndo:bool=false
   ) -> (then, ReturnValue:bool)
  ]}
 EventGraph {nodes [
   Picker Mouse Exit() {is=title K2Node_CustomEvent}
   Picker Mouse Enter() {is=title K2Node_CustomEvent}
   Event PreConstruct()
   Event Construct()
   Event Tick()
  ]}
 Find Mirror Control {nodes [
   n1 = Find Mirror Control!() -> (then, Name:Name) {is=title K2Node_FunctionEntry}
   n2 = Get()
   n3 = "To String (Name)"(InName=@n1)
   n4 = Contains(SearchIn=@n3, Substring=_l_, bUseCase=true, bSearchFromEnd=true)
   n5 = String To Name(InString=Replace(SourceString=@n3, From=SelectString(A=_l_, B=_r_, bPickA=@n4), To=SelectString(A=_r_, B=_l_, bPickA=@n4)))
   n6 = AND Boolean(
    A=Contains!(
     self:Rig Hierarchy Object Reference=GetHierarchy!(self:Control Rig Object Reference=@n2.Control Rig Binding_ControlRig) -> (ReturnValue:Rig Hierarchy Object Reference)
     InKey:Rig Element Key Structure=Make Rig Element Key!(Type:ERigElementType Enum=Control, Name:Name=@n5) -> (RigElementKey:Rig Element Key Structure) {is=title K2Node_MakeStruct}
    ) -> (ReturnValue)
    B="Not Equal (Name)"(A=Get Name() -> (Name:Name=None), B=@n5) -> (ReturnValue=~)
   )
   > n7 = Branch(Condition=@n6)
   > Return Node!(execute, Exists:bool=@n6, Mirror Item:Rig Element Key Structure=Make Rig Element Key(Type=Control, Name=@n5))
   n8 = GetHierarchy(self=Get().Control Rig Binding_ControlRig)
   n9 = Make Rig Element Key(Type=Control, Name=Get Name() -> (Name=None))
   n10 = Break Rig Control Settings!(RigControlSettings:"Rig Control Settings Structure (by ref)"=Break Rig Control Element!(RigControlElement:"Rig Control Element Structure (by ref)"=Find Control!(self:Rig Hierarchy Object Reference=@n8, InKey:"Rig Element Key Structure (by ref)"=@n9) -> (ReturnValue:Rig Control Element Structure)) -> (
     PreferredEulerAngles:Rig Preferred Euler Angles Structure
     TopologyControlSettings:Control Settings Topology Structure
     Key:Rig Element Key Structure
     Index:int
     SubIndex:int
     CreatedAtInstructionIndex:int
     SpawnIndex:int
     Settings:Rig Control Settings Structure
    ) {is=title K2Node_BreakStruct}.Settings) -> (
    AnimationType:ERigControlAnimationType Enum
    ControlType:ERigControlType Enum
    DisplayName:Name
    PrimaryAxis:ERigControlAxis Enum
    LimitEnabled:Array of Rig Control Limit Enabled Structures
    bDrawLimits:bool=~
    MinimumValue:Rig Control Value Structure
    MaximumValue:Rig Control Value Structure
    bShapeVisible:bool=~
    ShapeVisibility:ERigControlVisibility Enum
    ShapeName:Name
    ShapeColor:Linear Color Structure
    bIsTransientControl:bool=~
    ControlEnum:Enum Object Reference
    Customization:Rig Control Element Customization Structure
    DrivenControls:Array of Rig Element Key Structures
    bGroupWithParentControl:bool=~
    bRestrictSpaceSwitching:bool=~
    FilteredChannels:Array of ERigControlTransformChannel Enums
    PreferredRotationOrder:EEulerRotationOrder Enum
    bUsePreferredRotationOrder:bool=~
    ShapeTransform:Transform
   ) {is=title K2Node_BreakStruct}
   n11 = Break Rig Element Key!(RigElementKey:"Rig Element Key Structure (by ref)"=GetDefaultParent!(self:Rig Hierarchy Object Reference=@n8, InKey:Rig Element Key Structure=@n9) -> (ReturnValue:Rig Element Key Structure)) -> (Type:ERigElementType Enum, Name:Name) {is=title K2Node_BreakStruct}
   n12 = Branch(
    execute=@n7.else
    Condition=AND Boolean(
     A="Equal (Enum)"(A:ERigControlAnimationType Enum=@n10, B:ERigControlAnimationType Enum=AnimationChannel)
     B="Equal (Enum)"(A:ERigElementType Enum=@n11, B:ERigElementType Enum=Control)
    )
   )
   n13 = Find Mirror Control!(Name:Name=@n11.Name) -> (Exists:bool, Mirror Item:Rig Element Key Structure) {is=title K2Node_CallFunction}
   > n14 = Branch(Condition=@n13)
   n15 = Get()
   > n16 = For Each Loop(Array:Array of Rig Element Key Structures=GetChildren!(
     self:Rig Hierarchy Object Reference=GetHierarchy(self=@n15.Control Rig Binding_ControlRig)
     InKey:Rig Element Key Structure=@n13.Mirror Item
     bRecursive:bool=false
    ) -> (ReturnValue:Array of Rig Element Key Structures)) -> (Array Element:Rig Element Key Structure)
   Return Node(execute=[@n12.else, @n14.else, @n16.Completed], Exists=false)
   n17 = Break Rig Element Key(RigElementKey=@n16)
   n18 = Get Name() -> (Name=None)
   > Branch(Condition=StartsWith(SourceString="To String (Name)"(InName=@n17.Name), InPrefix="To String (Name)"(InName=@n18), SearchCase=CaseSensitive))
   > Return Node(Exists=true, Mirror Item=@n16)
   Get Name() -> (Name=None)
  ]}
 Get All Picker Buttons {nodes [
   Get All Picker Buttons() {is=title K2Node_FunctionEntry}
   > Clear(TargetArray:Array of EUW Picker Button Object References=Get Local Picker Buttons() -> (Local Picker Buttons:Array of EUW Picker Button Object References))
   > n1 = For Each Loop(Array:Array of Widget Object References=GetAllChildren(self=Get Primary Panel Widget() -> (Primary Panel Widget:Panel Widget Object Reference))) -> (Array Element:Widget Object Reference)
   Return Node!(execute=@n1.Completed, Picker Buttons:Array of EUW Picker Button Object References=Get Local Picker Buttons())
   > n2 = Cast To EUW_Picker_Button!(execute, Object=@n1) -> (then, CastFailed, AsEUW Picker Button:EUW Picker Button Object Reference) {is=title K2Node_DynamicCast}
   > Add(
    TargetArray:Array of EUW Picker Button Object References=Get Local Picker Buttons()
    NewItem:"EUW Picker Button Object Reference (by ref)"=@n2
   ) {is=title K2Node_CallArrayFunction}
  ]}
 Get Control Mirror Plane {nodes [
   Get Control Mirror Plane!() -> (then, Control:Name) {is=title K2Node_FunctionEntry}
   > Return Node!(
    execute
    ReturnValue:Enum_MirrorPlane Enum=Literal enum Enum_MirrorPlane!(Enum:Enum_MirrorPlane Enum=NewEnumerator0) -> (ReturnValue:Enum_MirrorPlane Enum) {is=title K2Node_EnumLiteral}
    Behavioral:bool=false
   )
  ]}
 Get Mirror Control Center {nodes [
   Get Mirror Control Center!() -> (then, Control:Name) {is=title K2Node_FunctionEntry}
   > Return Node!(execute, Transform:Transform=MakeTransform(Location="1,1,1", Rotation="-1,1,1"))
  ]}
 Mirror Control {nodes [
   Mirror Control() {is=title K2Node_FunctionEntry}
   > Clear!(
    execute
    TargetMap:Map of Rig Element Key Structures to Rig Control Value Structures=Get Mirror Control Values() -> (Mirror Control Values:Map of Rig Element Key Structures to Rig Control Value Structures)
   ) -> (then) {is=title K2Node_CallFunction}
   > n1 = For Each Loop(Array:Array of Names=CurrentControlSelection!(self:Control Rig Object Reference=Get().Control Rig Binding_ControlRig) -> (ReturnValue:Array of Names)) -> (Array Element:Name)
   n2 = Keys!(execute=@n1.Completed, TargetMap:Map of Rig Element Key Structures to Rig Control Value Structures=Get Mirror Control Values()) -> (then, Keys:Array of Rig Element Key Structures)
   > n3 = For Each Loop(Array:Array of Rig Element Key Structures=@n2) -> (Array Element:Rig Element Key Structure)
   n4 = Break Rig Element Key(RigElementKey=@n3)
   n5 = Find Control!(self=GetHierarchy(self=Get().Control Rig Binding_ControlRig), InKey_Type:ERigElementType Enum=Control, InKey_Name:Name=@n4.Name) -> (ReturnValue)
   > n6 = Switch on ERigControlType!(
    execute
    Selection:ERigControlType Enum=Break Rig Control Settings(RigControlSettings=Break Rig Control Element(RigControlElement=@n5).Settings) -> (bDrawLimits=~, bShapeVisible=~, bIsTransientControl=~, bGroupWithParentControl=~, bRestrictSpaceSwitching=~, bUsePreferredRotationOrder=~).ControlType
   ) -> (Bool:Exec, Float:Exec, Integer:Exec, Vector2D:Exec, Position:Exec, Scale:Exec, Rotator:Exec, EulerTransform:Exec, ScaleFloat:Exec) {is=title K2Node_SwitchEnum}
   n7 = Get Frame Number() -> (Frame Number:Frame Number Structure)
   n8 = Get Level Sequence() -> (Level Sequence:Level Sequence Object Reference)
   > n9 = GetLocalControlRigBool!(
    execute
    LevelSequence:Level Sequence Object Reference=@n8
    ControlRig:Control Rig Object Reference
    ControlName:Name=@n4.Name
    Frame:Frame Number Structure=@n7
    TimeUnit:EMovieSceneTimeUnit Enum=DisplayRate
   ) -> (then, ReturnValue:bool)
   > SetLocalControlRigBool!(
    execute
    LevelSequence:Level Sequence Object Reference=@n8
    ControlRig:Control Rig Object Reference=Get().Control Rig Binding_ControlRig
    ControlName:Name=@n4.Name
    Frame:Frame Number Structure=@n7
    Value:bool=@n9
    TimeUnit:EMovieSceneTimeUnit Enum=DisplayRate
    bSetKey:bool=true
   ) -> (then)
   n10 = Find(
    TargetMap:Map of Rig Element Key Structures to Rig Control Value Structures=Get Mirror Control Values()
    Key:"Rig Element Key Structure (by ref)"=@n3
   ) -> (Value:Rig Control Value Structure)
   n11 = GetEulerTransformFromControlValue!(execute=@n6.EulerTransform, InValue:Rig Control Value Structure=@n10) -> (then, ReturnValue:Euler Transform Structure)
   n12 = Get Control Mirror Plane!(Control:Name=@n4.Name) -> (ReturnValue:Enum_MirrorPlane Enum=NewEnumerator0, Behavioral:bool) {is=title K2Node_CallFunction}
   n13 = Mirror Transform!(Behavioral:bool=@n12.Behavioral, EulerTransform:"Euler Transform Structure (by ref)"=@n11, Mirror Plane:Enum_MirrorPlane Enum=@n12) -> (Mirror Transform:Euler Transform Structure)
   > MakeControlValueFromEulerTransform!(execute, InValue:Euler Transform Structure=@n13) -> (then, ReturnValue:Rig Control Value Structure)
   > SetLocalControlRigEulerTransform!(
    execute
    LevelSequence:Level Sequence Object Reference=Get Level Sequence()
    ControlRig:Control Rig Object Reference=Get().Control Rig Binding_ControlRig
    ControlName:Name=@n4.Name
    Frame:Frame Number Structure=Get Frame Number()
    Value:Euler Transform Structure=@n13
    TimeUnit:EMovieSceneTimeUnit Enum=DisplayRate
    bSetKey:bool=true
   ) -> (then)
   n14 = GetFloatFromControlValue!(execute=@n6.Float, InValue:Rig Control Value Structure=@n10) -> (then, ReturnValue:float)
   > SetLocalControlRigFloat!(
    execute
    LevelSequence:Level Sequence Object Reference=Get Level Sequence()
    ControlRig:Control Rig Object Reference=Get().Control Rig Binding_ControlRig
    ControlName:Name=@n4.Name
    Frame:Frame Number Structure=Get Frame Number()
    Value:float=@n14
    TimeUnit:EMovieSceneTimeUnit Enum=DisplayRate
    bSetKey:bool=true
   ) -> (then)
   n15 = GetIntFromControlValue!(execute=@n6.Integer, InValue:Rig Control Value Structure=@n10) -> (then, ReturnValue:int=0)
   > SetLocalControlRigInt!(
    execute
    LevelSequence:Level Sequence Object Reference=Get Level Sequence()
    ControlRig:Control Rig Object Reference=Get().Control Rig Binding_ControlRig
    ControlName:Name=@n4.Name
    Frame:Frame Number Structure=Get Frame Number()
    Value:int=@n15
    TimeUnit:EMovieSceneTimeUnit Enum=DisplayRate
    bSetKey:bool=true
   ) -> (then)
   n16 = GetVectorFromControlValue!(execute=@n6.Position, InValue:Rig Control Value Structure=@n10) -> (then, ReturnValue:Vector)
   n17 = Get Control Mirror Plane(Control=@n4.Name) -> (ReturnValue=NewEnumerator0)
   > SetLocalControlRigPosition!(
    execute
    LevelSequence:Level Sequence Object Reference=Get Level Sequence()
    ControlRig:Control Rig Object Reference=Get().Control Rig Binding_ControlRig
    ControlName:Name=@n4.Name
    Frame:Frame Number Structure=Get Frame Number()
    Value:Vector=Mirror Transform!(
     Behavioral=@n17.Behavioral
     EulerTransform=Make Euler Transform!(Location:Vector=@n16, Rotation:Rotator="0,0,0", Scale:Vector="1,1,1") -> (EulerTransform:Euler Transform Structure) {is=title K2Node_MakeStruct}
     Mirror Plane=@n17
    ) -> (Mirror Transform_Location:Vector=~, Mirror Transform_Rotation:Rotator=~, Mirror Transform_Scale:Vector=~)
    TimeUnit:EMovieSceneTimeUnit Enum=DisplayRate
    bSetKey:bool=true
   ) -> (then)
   n18 = GetRotatorFromControlValue!(execute=@n6.Rotator, InValue:Rig Control Value Structure=@n10) -> (then, ReturnValue:Rotator)
   n19 = Get Control Mirror Plane(Control=@n4.Name) -> (ReturnValue=NewEnumerator0)
   > SetLocalControlRigRotator!(
    execute
    LevelSequence:Level Sequence Object Reference=Get Level Sequence()
    ControlRig:Control Rig Object Reference=Get().Control Rig Binding_ControlRig
    ControlName:Name=@n4.Name
    Frame:Frame Number Structure=Get Frame Number()
    Value:Rotator=Mirror Transform(Behavioral=@n19.Behavioral, EulerTransform=Make Euler Transform(Location="0,0,0", Rotation=@n18, Scale="1,1,1"), Mirror Plane=@n19) -> (Mirror Transform_Location=~, Mirror Transform_Rotation=~, Mirror Transform_Scale=~).Mirror Transform_Rotation
    TimeUnit:EMovieSceneTimeUnit Enum=DisplayRate
    bSetKey:bool=true
   ) -> (then)
   Set Current Control(execute=@n1.LoopBody, Current Control:Name=@n1) -> (Output_Get:Name=None)
   > n20 = GetFocusedLevelSequence!(execute) -> (then, ReturnValue:Level Sequence Object Reference)
   > Set Level Sequence(Level Sequence:Level Sequence Object Reference=@n20) -> (Output_Get:Level Sequence Object Reference)
   > n21 = GetCurrentTime!(execute) -> (then, ReturnValue:int=0)
   > Set Frame Number(Frame Number:Frame Number Structure=Make Frame Number!(Value:int=@n21) -> (FrameNumber:Frame Number Structure) {is=title K2Node_MakeStruct}) -> (Output_Get:Frame Number Structure)
   n22 = Find Mirror Control(Name=Get Current Control() -> (Current Control:Name=None))
   > n23 = Branch(Condition=@n22)
   n24 = Make Rig Element Key(Type=Control, Name=Get Current Control() -> (Current Control=None))
   n25 = GetControlValue!(
    self:Rig Hierarchy Object Reference=GetHierarchy(self=Get().Control Rig Binding_ControlRig)
    InKey:Rig Element Key Structure=@n24
    InValueType:ERigControlValueType Enum=Current
   ) -> (ReturnValue:Rig Control Value Structure)
   > Add!(
    execute
    TargetMap:Map of Rig Element Key Structures to Rig Control Value Structures=Get Mirror Control Values()
    Key:"Rig Element Key Structure (by ref)"=@n22.Mirror Item
    Value:"Rig Control Value Structure (by ref)"=@n25
   ) -> (then) {is=title K2Node_CallFunction}
   Add(execute=@n23.else, TargetMap=Get Mirror Control Values(), Key=@n24, Value=@n25)
   n26 = Get Control Mirror Plane(Control=Get Current Control() -> (Current Control=None)) -> (ReturnValue=NewEnumerator0)
   Mirror Transform!(Behavioral=@n26.Behavioral, EulerTransform, Mirror Plane=@n26) -> (Mirror Transform:Euler Transform Structure)
   Reroute Node(InputPin:Rig Element Key Structure=@n3) -> (OutputPin:Rig Element Key Structure)
  ]}
 Mirror Selected Controls {nodes [
   Mirror Selected Controls() {is=title K2Node_FunctionEntry}
   > n1 = GetFocusedLevelSequence()
   > n2 = GetCurrentTime() -> (ReturnValue=0)
   n3 = Break Control Rig Sequencer Binding Proxy!(ControlRigSequencerBindingProxy:"Control Rig Sequencer Binding Proxy Structure (by ref)"=Get Control Rig Binding() -> (Control Rig Binding:Control Rig Sequencer Binding Proxy Structure)) -> (
    Proxy:Movie Scene Binding Proxy Structure
    ControlRig:Control Rig Object Reference
    Track:Movie Scene Control Rig Parameter Track Object Reference
   ) {is=title K2Node_BreakStruct}
   > n4 = For Each Loop(Array:Array of Names=CurrentControlSelection(self=@n3.ControlRig)) -> (Array Element:Name)
   n5 = Make Frame Number(Value=@n2)
   > n6 = GetControlRigWorldTransform!(
    execute
    LevelSequence:Level Sequence Object Reference=@n1
    ControlRig:Control Rig Object Reference=@n3.ControlRig
    ControlName:Name=@n4
    Frame:Frame Number Structure=@n5
    TimeUnit:EMovieSceneTimeUnit Enum=DisplayRate
   ) -> (then, ReturnValue:Transform)
   > SetControlRigWorldTransform!(
    execute
    LevelSequence:Level Sequence Object Reference=@n1
    ControlRig:Control Rig Object Reference=@n3.ControlRig
    ControlName:Name=@n4
    Frame:Frame Number Structure=@n5
    WorldTransform:"Transform (by ref)"=@n6
    TimeUnit:EMovieSceneTimeUnit Enum=DisplayRate
    bSetKey:bool=true
   ) -> (then)
  ]}
 Mirror Transform {nodes [
   n1 = Mirror Transform!() -> (then, Behavioral:bool=~, EulerTransform:"Euler Transform Structure (by ref)", Mirror Plane:Enum_MirrorPlane Enum=NewEnumerator0) {is=title K2Node_FunctionEntry}
   n2 = Break Euler Transform!(EulerTransform:"Euler Transform Structure (by ref)"=@n1.EulerTransform) -> (Location:Vector=~, Rotation:Rotator=~, Scale:Vector=~) {is=title K2Node_BreakStruct}
   n3 = "ToQuaternion (Rotator)"!(InRot:Rotator=@n2.Rotation) -> (ReturnValue_X:float, ReturnValue_Y:float, ReturnValue_Z:float, ReturnValue_W:float)
   n4 = "ToRotator (Quat)"!(
    Q_X:float=float * float(A=@n3, B=-1) {is=title K2Node_PromotableOperator}
    Q_Y:float=float * float(
     A=@n3.ReturnValue_Y
     B=Select!(NewEnumerator0:float=1, NewEnumerator1:float=1, NewEnumerator2:float=-1, Index:Enum_MirrorPlane Enum=@n1.Mirror Plane) -> (ReturnValue:float)
    )
    Q_Z:float=float * float(A=@n3.ReturnValue_Z, B=Select(NewEnumerator0=1, NewEnumerator1=-1, NewEnumerator2=1, Index=@n1.Mirror Plane))
    Q_W:float=float * float(A=@n3.ReturnValue_W, B=Select(NewEnumerator0=-1, NewEnumerator1=1, NewEnumerator2=1, Index=@n1.Mirror Plane))
   ) -> (ReturnValue:Rotator)
   n5 = Vector_MirrorByPlane!(
    A:Vector=@n2
    InPlane:"Plane Structure (by ref)"=Select(
     NewEnumerator0:"Plane Structure (by ref)"=MakePlaneFromPointAndNormal!(Point:Vector="0, 0, 0", Normal:Vector="1,0, 0") -> (ReturnValue:Plane Structure)
     NewEnumerator1:"Plane Structure (by ref)"=MakePlaneFromPointAndNormal(Point="0, 0, 0", Normal="0,1, 0")
     NewEnumerator2:"Plane Structure (by ref)"=MakePlaneFromPointAndNormal(Point="0, 0, 0", Normal="0,0,1")
     Index=@n1.Mirror Plane
    ) -> (ReturnValue:"Plane Structure (by ref)")
   ) -> (ReturnValue:Vector)
   n6 = TransformLocation!(T:"Transform (by ref)"=MakeTransform(Rotation="0, 0,180"), Location:Vector=@n5) -> (ReturnValue:Vector)
   > Return Node!(
    execute
    Mirror Transform:Euler Transform Structure=Make Euler Transform(
     Location=Select!(Option 0:Vector=@n5, Option 1:Vector=@n6, Index:bool=@n1) -> (ReturnValue:Vector)
     Rotation=Select(Option 0:Rotator=@n4, Option 1:Rotator=@n2.Rotation, Index=@n1) -> (ReturnValue:Rotator)
     Scale=@n2.Scale
    )
   )
  ]}
 Select All Control {nodes [
   Select All Control() {is=title K2Node_FunctionEntry}
   n1 = Get()
   > ClearControlSelection(self=@n1.Control Rig Binding_ControlRig, bSetupUndo=false)
   > n2 = For Each Loop!(
    Exec
    Array:Array of Rig Element Key Structures=Get Controls!(self:Rig Hierarchy Object Reference=GetHierarchy(self=@n1.Control Rig Binding_ControlRig), bTraverse:bool=true) -> (ReturnValue:Array of Rig Element Key Structures) {is=title K2Node_CallFunction}
   ) -> (LoopBody, Array Element_Type:ERigElementType Enum, Array Element_Name:Name, Array Index, Completed)
   > Branch(Condition=EndsWith(SourceString="To String (Name)"(InName=@n2.Array Element_Name), InSuffix=_ctrl))
   > SelectControl!(
    execute
    self:Control Rig Object Reference=Get().Control Rig Binding_ControlRig
    InControlName:"Name (by ref)"=@n2.Array Element_Name
    bSelect:bool=true
    bSetupUndo:bool=false
   ) -> (then)
  ]}
 Select Control {nodes [
   n1 = Select Control!() -> (then, Control:Button Object Reference) {is=title K2Node_FunctionEntry}
   n2 = Get()
   > ClearControlSelection(self=@n2.Control Rig Binding_ControlRig, bSetupUndo=false)
   > SelectControl(self=@n2.Control Rig Binding_ControlRig, InControlName=String To Name(InString=GetDisplayName(Object=@n1)), bSelect=true, bSetupUndo=false)
  ]}
]
variables [
 Control Rig Binding {
  type=ControlRigSequencerBindingProxy
  default {controlRig=null, track=null, proxy {bindingId=#1, sequence=null}}
 }
 Primary Panel Widget {type=PanelWidget, default=null}
]
widgets {info {rootWidgetClass=null, widgetCount=0}, widgets []}
