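# Editor utility widget that draws a translucent marquee selection rectangle, in the AnimationPicker folder of AnimInEdTools.
# Start Drag stores the start location and shows the widget; End Drag hides it and returns start and end in local coordinates.
# While visible with Update enabled, Tick pins the rectangle at the start and render-scales it to the mouse from a top-left pivot, rather than resizing it.
# An onDragCompleted dispatcher is declared but never called here.
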EditorUtilityWidgetBlueprint /AnimInEdTools/AnimationPicker/EUW_Marquee_Selection
graphs [
 End Drag {nodes [
   n1 = End Drag!() -> (then, Input:"Pointer Event Structure (by ref)") {is=title K2Node_FunctionEntry}
   > SetVisibility(InVisibility=Hidden)
   n2 = GetCachedGeometry!(self:Widget Object Reference) -> (ReturnValue:"Geometry Structure (by ref)")
   > Return Node!(
    execute
    Start Positin:Vector 2D Structure="Absolute to Local (Coordinates)"!(
     Geometry:"Geometry Structure (by ref)"=@n2
     AbsoluteCoordinate:Vector 2D Structure=Get Current Location() -> (Current Location:Vector 2D Structure)
    ) -> (ReturnValue:Vector 2D Structure)
    End Position:Vector 2D Structure="Absolute to Local (Coordinates)"(
     Geometry=@n2
     AbsoluteCoordinate=Get Screen Space Position!(Input:"Pointer Event Structure (by ref)"=@n1) -> (ReturnValue:Vector 2D Structure) {is=title K2Node_CallFunction}
    )
   )
  ]}
 EventGraph {nodes [
   Event Tick()
   > Branch(Condition=AND Boolean(A=Get Update() -> (Update:bool), B=IsVisible()))
   > Update Position()
   Event PreConstruct()
   Event Construct()
  ]}
 Start Drag {nodes [
   n1 = Start Drag!() -> (then, current_location:Vector 2D Structure) {is=title K2Node_FunctionEntry}
   > Set Current Location(Current Location:Vector 2D Structure=@n1) -> (Output_Get:Vector 2D Structure)
   > SetVisibility(InVisibility=Visible)
  ]}
 Update Position {nodes [
   Update Position() {is=title K2Node_FunctionEntry}
   n1 = GetCachedGeometry()
   n2 = "Absolute to Local (Coordinates)"(Geometry=@n1, AbsoluteCoordinate=Get Current Location())
   n3 = SlotAsCanvasSlot!(Widget:Widget Object Reference=Get drag_rectangle() -> (drag_rectangle:Border Object Reference)) -> (ReturnValue:Canvas Panel Slot Object Reference)
   > SetPosition!(execute, self:Canvas Panel Slot Object Reference=@n3, InPosition:Vector 2D Structure=@n2) -> (then)
   > n4 = GetMousePositionOnPlatform!(execute) -> (then, ReturnValue:Vector 2D Structure)
   n5 = "Absolute to Local (Coordinates)"(Geometry=@n1, AbsoluteCoordinate=@n4)
   > SetRenderScale!(
    execute
    self:Widget Object Reference=Get drag_rectangle()
    Scale:Vector 2D Structure=vector2d / vector2d(
     A:Vector 2D Structure=vector2d - vector2d!(A:Vector 2D Structure=@n5, B:Vector 2D Structure=@n2) -> (ReturnValue:Vector 2D Structure)
     B:Vector 2D Structure=GetSize!(self:Canvas Panel Slot Object Reference=@n3) -> (ReturnValue:Vector 2D Structure)
    ) -> (ReturnValue:Vector 2D Structure) {is=title K2Node_PromotableOperator}
   ) -> (then)
  ]}
 onDragCompleted {nodes [onDragCompleted!() -> (then, start_position:Vector 2D Structure, end_position:Vector 2D Structure) {is=title K2Node_FunctionEntry}]}
]
variables [
 Current Location {type=Vector2D, default=<0 0>}
 onDragCompleted {default=()}
 Update {type=boolean, default=true}
]
widgets {
 info {rootWidgetClass=&Border, widgetCount=3}
 widgets [
  Border_0 {widgetClassPath=&Border, widget=Border_0, properties {brushColor=<rgba 1 1 1 0>}}
  CanvasPanel_0 {widgetClassPath=&CanvasPanel, parent=Border_0, widget=CanvasPanel_0, slot {class=BorderSlot}}
  drag_rectangle {
   widgetClassPath=&Border
   bIsVariable=true
   parent=CanvasPanel_0
   widget=drag_rectangle
   properties {visibility=HitTestInvisible, brushColor=<rgba 0.6719 0.7022 1 0.5>, renderTransformPivot=<0 0>}
   slot {class=CanvasPanelSlot, properties {layoutData {alignment=<0 0>, anchors {maximum=<0 0>, minimum=<0 0>}, offsets=<ltrb 0 0 1920 1080>}}}
  }
 ]
}
