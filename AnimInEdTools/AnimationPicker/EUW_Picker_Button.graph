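# Editor Utility Widget for a labeled Animation Picker button element, derived from EUW_Picker_Element.
# Pressing it passes itself to the picker and fires a custom click delegate; releasing also notifies the picker.
# Update Selection Display rebuilds the button style from shape, images and tint, brightening hover and pressed states, using the pressed brush while selected, and applies label styling.
# Copy Button Attributes clones another button's images, tint, shape and label, then refreshes deselected.
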EditorUtilityWidgetBlueprint /AnimInEdTools/AnimationPicker/EUW_Picker_Button
graphs [
 Copy Button Attributes {nodes [
   Copy Button Attributes!() -> (then, Input Button:EUW Picker Button Object Reference) {is=title K2Node_FunctionEntry}
   > n1 = Sequence!(execute) -> (then_0, then_1, then_2:Exec, then_3:Exec, then_4:Exec, then_5:Exec, then_6:Exec)
   n2 = Get Normal Image!(self:EUW Picker Button Object Reference=Get Input Button() -> (Input Button:EUW Picker Button Object Reference)) -> (Normal Image:Texture Object Reference)
   > Is Valid(InputObject=@n2)
   > Set Normal Image(Normal Image:Texture Object Reference=@n2) -> (Output_Get:Texture Object Reference)
   n3 = Get Hovered Image!(self:EUW Picker Button Object Reference=Get Input Button()) -> (Hovered Image:Texture Object Reference)
   Is Valid(exec=@n1.then_1, InputObject=@n3)
   > Set Hovered Image(Hovered Image:Texture Object Reference=@n3) -> (Output_Get:Texture Object Reference)
   n4 = Get Pressed Image!(self:EUW Picker Button Object Reference=Get Input Button()) -> (Pressed Image:Texture Object Reference)
   Is Valid(exec=@n1.then_2, InputObject=@n4)
   > Set Pressed Image(Pressed Image:Texture Object Reference=@n4) -> (Output_Get:Texture Object Reference)
   Set Normal Tint(
    execute=@n1.then_3
    Normal Tint:Slate Color Structure=Get Normal Tint!(self:EUW Picker Button Object Reference=Get Input Button()) -> (Normal Tint:Slate Color Structure)
   ) -> (Output_Get:Slate Color Structure)
   Set Draw Style(
    execute=@n1.then_4
    Draw Style:Enum_PickerButtonShapes Enum=Get Draw Style!(self:EUW Picker Button Object Reference=Get Input Button()) -> (Draw Style:Enum_PickerButtonShapes Enum=NewEnumerator0)
   ) -> (Output_Get:Enum_PickerButtonShapes Enum=NewEnumerator0)
   Set Label Font Size(
    execute=@n1.then_5
    Label Font Size:int=Get Label Font Size!(self:EUW Picker Button Object Reference=Get Input Button()) -> (Label Font Size:int=0)
   ) -> (Output_Get:int=0)
   n5 = Get Input Button()
   > Set Label Text(Label Text:Text=Get Label Text!(self:EUW Picker Button Object Reference=@n5) -> (Label Text:Text)) -> (Output_Get:Text)
   > Set Label Color(Label Color:Slate Color Structure=Get Label Color!(self:EUW Picker Button Object Reference=@n5) -> (Label Color:Slate Color Structure)) -> (Output_Get:Slate Color Structure)
   Update Selection Display!(execute=@n1.then_6, self:BPI Selectable Picker Elements Interface, Selected:bool=false) -> (then, Out Selected:bool)
  ]}
 Custom On Clicked Event {nodes [Custom On Clicked Event() {is=title K2Node_FunctionEntry}]}
 EventGraph {nodes [
   "On Released (button)"() {is=title K2Node_ComponentBoundEvent}
   > Picker Button Released(self:Animation Picker Object Reference=Get Anim Picker EUW() -> (Anim Picker EUW:Animation Picker Object Reference))
   "On Pressed (button)"() {is=title K2Node_ComponentBoundEvent}
   > Picker Element Clicked!(execute, self:Animation Picker Object Reference=Get Anim Picker EUW(), Picker Element:EUW Picker Element Object Reference=Self-Reference()) -> (then)
   > Call Custom on Clicked Event!(execute:Exec, self:EUW Picker Button Object Reference) -> (then:Exec) {is=title K2Node_CallDelegate}
   Event PreConstruct()
   > Update Selection Display(Selected=false)
  ]}
 Make Slate Brush {nodes [
   n1 = Make Slate Brush!() -> (then, Draw As:Enum_PickerButtonShapes Enum=NewEnumerator0, Image:Object Reference, Tint:Slate Color Structure) {is=title K2Node_FunctionEntry}
   > n2 = Switch on Enum_PickerButtonShapes!(execute, Selection:Enum_PickerButtonShapes Enum=@n1) -> (NewEnumerator0:Exec, NewEnumerator1:Exec, NewEnumerator2:Exec, NewEnumerator3:Exec, NewEnumerator4:Exec) {is=title K2Node_SwitchEnum}
   n3 = Cast To Texture2D(execute=@n2.NewEnumerator4, Object=Get Image() -> (Image:Object Reference))
   n4 = GetSizeX!(self:Texture 2D Object Reference=@n3) -> (ReturnValue:int=0)
   n5 = GetSizeY!(self:Texture 2D Object Reference=@n3) -> (ReturnValue:int=0)
   > Return Node!(
    execute
    SlateBrush:Slate Brush Structure=Make Slate Brush!(
     TintColor:Slate Color Structure=Make Slate Color(SpecifiedColor=(R=1,G=1,B=1,A=1))
     DrawAs:ESlateBrushDrawType Enum=Image
     Tiling:ESlateBrushTileType Enum=NoTile
     Mirroring:ESlateBrushMirrorType Enum=NoMirror
     ImageSize_X:double="To Float (Integer)"(InInt=@n4)
     ImageSize_Y:double="To Float (Integer)"(InInt=@n5)
     Margin:Margin Structure=(Left=0,Top=0,Right=0,Bottom=0)
     ResourceObject:Object Reference=Get Image()
     OutlineSettings:Slate Brush Outline Settings Structure=(CornerRadii=(X=0,Y=0,Z=0,W=0),Color=(SpecifiedColor=(R=0,G=0,B=0,A=0),ColorUseRule=UseColor_Specified),Width=0,RoundingType=HalfHeightRadius,bUseBrushTransparency=False)
    ) -> (SlateBrush:Slate Brush Structure) {is=title K2Node_MakeStruct}
   )
   n6 = MakeVector4!(X:double=10, Y:double=10, Z:double=10, W:double=10) -> (ReturnValue:Vector 4 Structure)
   Return Node(
    execute=@n2.NewEnumerator2
    SlateBrush=Make Slate Brush(
     TintColor=Get Tint() -> (Tint:Slate Color Structure)
     DrawAs=RoundedBox
     Tiling=NoTile
     Mirroring=NoMirror
     ImageSize_X=0
     ImageSize_Y=0
     Margin=(Left=0,Top=0,Right=0,Bottom=0)
     OutlineSettings=Make Slate Brush Outline Settings!(
      CornerRadii:Vector 4 Structure=@n6
      Color:Slate Color Structure=(SpecifiedColor=(R=0,G=0,B=0,A=0),ColorUseRule=UseColor_Specified)
      Width:float=0
      RoundingType:ESlateBrushRoundingType Enum=FixedRadius
      bUseBrushTransparency:bool=False
     ) -> (SlateBrushOutlineSettings:Slate Brush Outline Settings Structure) {is=title K2Node_MakeStruct}
    )
   )
   Return Node(
    execute=@n2.NewEnumerator0
    SlateBrush=Make Slate Brush(
     TintColor=Get Tint()
     DrawAs=Box
     Tiling=NoTile
     Mirroring=NoMirror
     ImageSize_X=0
     ImageSize_Y=0
     Margin=(Left=0,Top=0,Right=0,Bottom=0)
     OutlineSettings=(CornerRadii=(X=0,Y=0,Z=0,W=0),Color=(SpecifiedColor=(R=0,G=0,B=0,A=0),ColorUseRule=UseColor_Specified),Width=0,RoundingType=HalfHeightRadius,bUseBrushTransparency=False)
    )
   )
   n7 = MakeVector4(X=5, Y=5, Z=5, W=5)
   Return Node(
    execute=@n2.NewEnumerator1
    SlateBrush=Make Slate Brush(
     TintColor=Get Tint()
     DrawAs=RoundedBox
     Tiling=NoTile
     Mirroring=NoMirror
     ImageSize_X=0
     ImageSize_Y=0
     Margin=(Left=0,Top=0,Right=0,Bottom=0)
     OutlineSettings=Make Slate Brush Outline Settings(
      CornerRadii=@n7
      Color=(SpecifiedColor=(R=0,G=0,B=0,A=0),ColorUseRule=UseColor_Specified)
      Width=0
      RoundingType=FixedRadius
      bUseBrushTransparency=False
     )
    )
   )
   n8 = MakeVector4(X=10, Y=10, Z=10, W=10)
   Return Node(
    execute=@n2.NewEnumerator3
    SlateBrush=Make Slate Brush(
     TintColor=Get Tint()
     DrawAs=RoundedBox
     Tiling=NoTile
     Mirroring=NoMirror
     ImageSize_X=0
     ImageSize_Y=0
     Margin=(Left=0,Top=0,Right=0,Bottom=0)
     OutlineSettings=Make Slate Brush Outline Settings(
      CornerRadii=@n8
      Color=(SpecifiedColor=(R=0,G=0,B=0,A=0),ColorUseRule=UseColor_Specified)
      Width=0
      RoundingType=HalfHeightRadius
      bUseBrushTransparency=False
     )
    )
   )
  ]}
 Update Selection Display {nodes [
   n1 = Update Selection Display!() -> (then, Selected:bool) {is=title K2Node_FunctionEntry}
   > Parent: Update Selection Display!(execute, Selected:bool=@n1) -> (then, Out Selected:bool) {is=title K2Node_CallParentFunction}
   n2 = Get Normal Tint!() -> (Normal Tint)
   > n3 = Make Slate Brush!(
    execute
    self
    Draw As:Enum_PickerButtonShapes Enum=Get Draw Style!() -> (Draw Style=NewEnumerator0)
    Image:Object Reference=Get Normal Image!() -> (Normal Image)
    Tint:Slate Color Structure=@n2
   ) -> (then, SlateBrush:Slate Brush Structure) {is=title K2Node_CallFunction}
   n4 = Break Slate Color!(SlateColor:"Slate Color Structure (by ref)"=@n2) -> (SpecifiedColor:Linear Color Structure, ColorUseRule:ESlateColorStylingMode Enum) {is=title K2Node_BreakStruct}
   n5 = MakeColor!(R:float=0.5, G:float=0.5, B:float=0.5, A:float=0) -> (ReturnValue:Linear Color Structure)
   > n6 = Make Slate Brush(
    Draw As=Get Draw Style() -> (Draw Style=NewEnumerator0)
    Image=Get Hovered Image!() -> (Hovered Image)
    Tint=Make Slate Color(SpecifiedColor=LinearColor + LinearColor(A:Linear Color Structure=@n4, B:Linear Color Structure=@n5) -> (ReturnValue:Linear Color Structure) {is=title K2Node_PromotableOperator})
   )
   n7 = Break Slate Color(SlateColor=@n2)
   n8 = MakeColor(R=0.8, G=0.8, B=0.8, A=0)
   > n9 = Make Slate Brush(
    Draw As=Get Draw Style() -> (Draw Style=NewEnumerator0)
    Image=Get Pressed Image!() -> (Pressed Image)
    Tint=Make Slate Color(SpecifiedColor=LinearColor + LinearColor(A=@n7, B=@n8))
   )
   n10 = Get Selected() -> (Selected:bool)
   n11 = Get Selected()
   > SetStyle!(
    execute
    self:Button Object Reference=Get button() -> (button:Button Object Reference)
    InStyle:"Button Style Structure (by ref)"=Make Button Style!(
     Normal:Slate Brush Structure=Select(Option 0:Slate Brush Structure=@n3, Option 1:Slate Brush Structure=@n9, Index=@n11) -> (ReturnValue:Slate Brush Structure)
     Hovered:Slate Brush Structure=Select(Option 0=@n6, Option 1=@n9, Index=@n10)
     Pressed:Slate Brush Structure=@n9
     Disabled:Slate Brush Structure=(TintColor=(SpecifiedColor=(R=1,G=1,B=1,A=1),ColorUseRule=UseColor_Specified),DrawAs=NoDrawType,Tiling=NoTile,Mirroring=NoMirror,ImageType=NoImage,ImageSize=(X=0,Y=0),Margin=(Left=0,Top=0,Right=0,Bottom=0),ResourceObject=None,OutlineSettings=(CornerRadii=(X=0,Y=0,Z=0,W=0),Color=(SpecifiedColor=(R=0,G=0,B=0,A=0),ColorUseRule=UseColor_Specified),Width=0,RoundingType=HalfHeightRadius,bUseBrushTransparency=False),UVRegion=(Min=(X=0,Y=0),Max=(X=0,Y=0),bIsValid=False),bIsDynamicallyLoaded=False,ResourceName="")
     NormalForeground:Slate Color Structure=(SpecifiedColor=(R=1,G=0,B=1,A=1),ColorUseRule=UseColor_Foreground)
     HoveredForeground:Slate Color Structure=(SpecifiedColor=(R=1,G=0,B=1,A=1),ColorUseRule=UseColor_Foreground)
     PressedForeground:Slate Color Structure=(SpecifiedColor=(R=1,G=0,B=1,A=1),ColorUseRule=UseColor_Foreground)
     DisabledForeground:Slate Color Structure=(SpecifiedColor=(R=1,G=0,B=1,A=1),ColorUseRule=UseColor_Foreground)
     NormalPadding:Margin Structure=(Left=0,Top=0,Right=0,Bottom=0)
     PressedPadding:Margin Structure=(Left=0,Top=0,Right=0,Bottom=0)
     PressedSlateSound:Slate Sound Structure=(ResourceObject=None)
     ClickedSlateSound:Slate Sound Structure=(ResourceObject=None)
     HoveredSlateSound:Slate Sound Structure=(ResourceObject=None)
    ) -> (ButtonStyle:Button Style Structure) {is=title K2Node_MakeStruct}
   ) -> (then)
   n12 = Get Label Font Size!() -> (Label Font Size=0)
   n13 = Get LabelText() -> (LabelText:Text Object Reference)
   > Set Font!(
    execute
    Font:Slate Font Info Structure=MakeSlateFontInfo!(
     FontObject:Object Reference=&Roboto
     FontMaterial:Object Reference
     OutlineSettings:Font Outline Settings Structure=(OutlineSize=0,bMiteredCorners=False,bSeparateFillAlpha=False,bApplyOutlineToDropShadows=False,OutlineMaterial=None,OutlineColor=(R=0,G=0,B=0,A=1))
     TypefaceFontName:Name=Bold
     Size:float="To Float (Integer)"(InInt=@n12)
     LetterSpacing:int=0
     SkewAmount:float=0
     bForceMonospaced:bool=false
     bMaterialIsStencil:bool=false
     MonospacedWidth:float=1
    ) -> (ReturnValue:Slate Font Info Structure)
    self:Text Object Reference=@n13
   ) -> (then, Output_Get:Slate Font Info Structure)
   > Set Text(Text=Get Label Text!() -> (Label Text), self=@n13)
   > SetColorAndOpacity(self=@n13, InColorAndOpacity=Get Label Color!() -> (Label Color))
   > Return Node!(execute, Out Selected:bool=Get Selected())
  ]}
]
variables [
 Control Rig Poses Paths {type="map<string>", default {}}
 Custom On Clicked Event {default=()}
 Draw Style {type=Enum_PickerButtonShapes, default=RoundedSquare}
 Hovered Image {type=Texture, default=null}
 Label Color {type=SlateColor, default {colorUseRule=UseColor_Specified, specifiedColor=<rgba 1 1 1 1>}}
 Label Font Size {type=integer, default=10}
 Label Text {type=string, default=""}
 Normal Image {type=Texture, default=null}
 Normal Tint {type=SlateColor, default {colorUseRule=UseColor_Specified, specifiedColor=<rgba 0.2745 0.2745 0.2745 1>}}
 Pressed Image {type=Texture, default=null}
]
widgets {
 info {
  parentClass=/AnimInEdTools/AnimationPicker/PickerElements/EUW_Picker_Element.EUW_Picker_Element_C
  rootWidgetClass=&CanvasPanel
  widgetCount=3
 }
 widgets [
  CanvasPanel_0 {widgetClassPath=&CanvasPanel, widget=CanvasPanel_0}
  button {
   widgetClassPath=&Button
   parent=CanvasPanel_0
   widget=button
   properties {
    cursor=ResizeSouthEast
    isFocusable=false
    widgetStyle {
     clickedSlateSound {resourceObject=null}
     disabled {
      bIsDynamicallyLoaded=false
      drawAs=RoundedBox
      imageType=NoImage
      mirroring=NoMirror
      resourceName=None
      resourceObject=null
      tiling=NoTile
      imageSize=<0 0>
      margin=<ltrb 0 0 0 0>
      outlineSettings {
       bUseBrushTransparency=false
       roundingType=FixedRadius
       width=1
       color {colorUseRule=UseColor_ColorTable, specifiedColor=<rgba 0 0 0 0>}
       cornerRadii=<4 4 4 4>
      }
      tintColor {colorUseRule=UseColor_Specified, specifiedColor=<rgba 0.0395 0.0395 0.0395 1>}
      uVRegion {bIsValid=false, max=<0 0>, min=<0 0>}
     }
     disabledForeground {colorUseRule=UseColor_Foreground, specifiedColor=<rgba 1 0 1 1>}
     hovered {
      bIsDynamicallyLoaded=false
      drawAs=Image
      imageType=NoImage
      mirroring=NoMirror
      resourceName=None
      resourceObject=null
      tiling=NoTile
      imageSize=<32 32>
      margin=<ltrb 0 0 0 0>
      outlineSettings {
       bUseBrushTransparency=false
       roundingType=HalfHeightRadius
       width=0
       color {colorUseRule=UseColor_ColorTable, specifiedColor=<rgba 0 0 0 0>}
       cornerRadii=<0 0 0 1>
      }
      tintColor {colorUseRule=UseColor_Specified, specifiedColor=<rgba 1 1 1 1>}
      uVRegion {bIsValid=false, max=<0 0>, min=<0 0>}
     }
     hoveredForeground {colorUseRule=UseColor_Foreground, specifiedColor=<rgba 1 0 1 1>}
     hoveredSlateSound {resourceObject=null}
     normal {
      bIsDynamicallyLoaded=false
      drawAs=Image
      imageType=NoImage
      mirroring=NoMirror
      resourceName=None
      resourceObject=null
      tiling=NoTile
      imageSize=<32 32>
      margin=<ltrb 0 0 0 0>
      outlineSettings {
       bUseBrushTransparency=false
       roundingType=HalfHeightRadius
       width=0
       color {colorUseRule=UseColor_ColorTable, specifiedColor=<rgba 0 0 0 0>}
       cornerRadii=<0 0 0 1>
      }
      tintColor {colorUseRule=UseColor_Specified, specifiedColor=<rgba 0.1562 0.1562 0.1562 1>}
      uVRegion {bIsValid=false, max=<0 0>, min=<0 0>}
     }
     normalForeground {colorUseRule=UseColor_Foreground, specifiedColor=<rgba 1 0 1 1>}
     normalPadding=<ltrb 0 0 0 0>
     pressed {
      bIsDynamicallyLoaded=false
      drawAs=Image
      imageType=NoImage
      mirroring=NoMirror
      resourceName=None
      resourceObject=null
      tiling=NoTile
      imageSize=<32 32>
      margin=<ltrb 0 0 0 0>
      outlineSettings {
       bUseBrushTransparency=false
       roundingType=HalfHeightRadius
       width=0
       color {colorUseRule=UseColor_ColorTable, specifiedColor=<rgba 0 0 0 0>}
       cornerRadii=<0 0 0 1>
      }
      tintColor {colorUseRule=UseColor_Specified, specifiedColor=<rgba 1 1 1 1>}
      uVRegion {bIsValid=false, max=<0 0>, min=<0 0>}
     }
     pressedForeground {colorUseRule=UseColor_Foreground, specifiedColor=<rgba 1 0 1 1>}
     pressedPadding=<ltrb 0 0 0 0>
     pressedSlateSound {resourceObject=null}
    }
   }
   slot {class=CanvasPanelSlot, properties {layoutData {alignment=<0 0>, anchors {maximum=<1 1>, minimum=<0 0>}, offsets=<ltrb 0 0 0 0>}}}
  }
  LabelText {
   widgetClassPath=&TextBlock
   bIsVariable=true
   parent=button
   widget=LabelText
   properties {
    autoWrapText=true
    justification=Center
    text=Label
    colorAndOpacity {colorUseRule=UseColor_Specified, specifiedColor=<rgba 0 0 0 1>}
    font {
     bForceMonospaced=false
     bMaterialIsStencil=false
     fontMaterial=null
     fontObject=&Roboto
     letterSpacing=0
     monospacedWidth=1
     size=10
     skewAmount=0
     typefaceFontName=Bold
     outlineSettings {
      bApplyOutlineToDropShadows=false
      bMiteredCorners=false
      bSeparateFillAlpha=false
      outlineMaterial=null
      outlineSize=0
      outlineColor=<rgba 0 0 0 1>
     }
    }
   }
   slot {class=ButtonSlot, properties {horizontalAlignment=HAlign_Fill, padding=<ltrb 0 0 0 0>}}
  }
 ]
}
