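# EUW_PaletteButton is an editor utility widget for a single color swatch button in the AnimInEdTools Animation Picker.
# On Construct it restyles PaletteButton from its Color variable, using that color for the normal tint and a reduced-alpha version for hover.
# Clicking broadcasts OnClickedButtonEventDispatcher with the normal tint read back from the button's style.
# A right-click dispatcher is declared but never called here, and two ThumbnailButton functions simply return that tint.
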EditorUtilityWidgetBlueprint /AnimInEdTools/AnimationPicker/EUW_PaletteButton
graphs [
 EventGraph {nodes [
   "On Clicked (PaletteButton)"() {is=title K2Node_ComponentBoundEvent}
   n1 = Get PaletteButton() -> (PaletteButton:Editor Utility Button Object Reference)
   > Call On Clicked Button Event Dispatcher!(
    execute:Exec
    self:EUW Palette Button Object Reference
    Color:Slate Color Structure=Break Button Style!(ButtonStyle:"Button Style Structure (by ref)"=Get WidgetStyle!(self:Button Object Reference=@n1) -> (WidgetStyle:Button Style Structure)) -> (
     Normal_TintColor:Slate Color Structure
     Normal_DrawAs:ESlateBrushDrawType Enum
     Normal_Tiling:ESlateBrushTileType Enum
     Normal_Mirroring:ESlateBrushMirrorType Enum
     Normal_ImageSize:Vector 2D Structure
     Normal_Margin:Margin Structure
     Normal_ResourceObject:Object Reference
     Normal_OutlineSettings:Slate Brush Outline Settings Structure
     Hovered:Slate Brush Structure
     Pressed:Slate Brush Structure
     Disabled:Slate Brush Structure
     NormalForeground:Slate Color Structure
     HoveredForeground:Slate Color Structure
     PressedForeground:Slate Color Structure
     DisabledForeground:Slate Color Structure
     NormalPadding:Margin Structure
     PressedPadding:Margin Structure
     PressedSlateSound:Slate Sound Structure
     ClickedSlateSound:Slate Sound Structure
     HoveredSlateSound:Slate Sound Structure
    ) {is=title K2Node_BreakStruct}
   ) -> (then:Exec) {is=title K2Node_CallDelegate}
   Event Construct()
   n2 = Get Color() -> (Color:Linear Color Structure)
   > SetStyle!(
    execute
    self:Button Object Reference=Get PaletteButton()
    InStyle_Normal_TintColor_SpecifiedColor:Linear Color Structure=@n2
    InStyle_Normal_TintColor_ColorUseRule:ESlateColorStylingMode Enum
    InStyle_Normal_DrawAs:ESlateBrushDrawType Enum=Box
    InStyle_Normal_Tiling:ESlateBrushTileType Enum=NoTile
    InStyle_Normal_Mirroring:ESlateBrushMirrorType Enum=NoMirror
    InStyle_Normal_ImageSize:Vector 2D Structure=(X=32,Y=32)
    InStyle_Normal_Margin:Margin Structure=(Left=0,Top=0,Right=0,Bottom=0)
    InStyle_Normal_ResourceObject:Object Reference
    InStyle_Normal_OutlineSettings:Slate Brush Outline Settings Structure=(CornerRadii=(X=0,Y=0,Z=0,W=0),Color=(SpecifiedColor=(R=0,G=0,B=0,A=0),ColorUseRule=UseColor_Specified),Width=0,RoundingType=HalfHeightRadius,bUseBrushTransparency=False)
    InStyle_Hovered_TintColor_SpecifiedColor:Linear Color Structure=LinearColor * LinearColor(A:Linear Color Structure=@n2, B:Linear Color Structure=(R=1,G=1,B=1,A=0.8)) -> (ReturnValue:Linear Color Structure) {is=title K2Node_PromotableOperator}
    InStyle_Hovered_TintColor_ColorUseRule:ESlateColorStylingMode Enum
    InStyle_Hovered_DrawAs:ESlateBrushDrawType Enum=Box
    InStyle_Hovered_Tiling:ESlateBrushTileType Enum=NoTile
    InStyle_Hovered_Mirroring:ESlateBrushMirrorType Enum=NoMirror
    InStyle_Hovered_ImageSize:Vector 2D Structure=(X=32,Y=32)
    InStyle_Hovered_Margin:Margin Structure=(Left=0,Top=0,Right=0,Bottom=0)
    InStyle_Hovered_ResourceObject:Object Reference
    InStyle_Hovered_OutlineSettings:Slate Brush Outline Settings Structure=(CornerRadii=(X=0,Y=0,Z=0,W=0),Color=(SpecifiedColor=(R=0,G=0,B=0,A=0),ColorUseRule=UseColor_Specified),Width=0,RoundingType=HalfHeightRadius,bUseBrushTransparency=False)
    InStyle_Pressed:Slate Brush Structure=(TintColor=(SpecifiedColor=(R=1,G=1,B=1,A=1),ColorUseRule=UseColor_Specified),DrawAs=Image,Tiling=NoTile,Mirroring=NoMirror,ImageType=NoImage,ImageSize=(X=32,Y=32),Margin=(Left=0,Top=0,Right=0,Bottom=0),ResourceObject=None,OutlineSettings=(CornerRadii=(X=0,Y=0,Z=0,W=0),Color=(SpecifiedColor=(R=0,G=0,B=0,A=0),ColorUseRule=UseColor_Specified),Width=0,RoundingType=HalfHeightRadius,bUseBrushTransparency=False),UVRegion=(Min=(X=0,Y=0),Max=(X=0,Y=0),bIsValid=False),bIsDynamicallyLoaded=False,ResourceName="")
    InStyle_Disabled:Slate Brush Structure=(TintColor=(SpecifiedColor=(R=1,G=1,B=1,A=1),ColorUseRule=UseColor_Specified),DrawAs=NoDrawType,Tiling=NoTile,Mirroring=NoMirror,ImageType=NoImage,ImageSize=(X=0,Y=0),Margin=(Left=0,Top=0,Right=0,Bottom=0),ResourceObject=None,OutlineSettings=(CornerRadii=(X=0,Y=0,Z=0,W=0),Color=(SpecifiedColor=(R=0,G=0,B=0,A=0),ColorUseRule=UseColor_Specified),Width=0,RoundingType=HalfHeightRadius,bUseBrushTransparency=False),UVRegion=(Min=(X=0,Y=0),Max=(X=0,Y=0),bIsValid=False),bIsDynamicallyLoaded=False,ResourceName="")
    InStyle_NormalForeground:Slate Color Structure=(SpecifiedColor=(R=1,G=0,B=1,A=1),ColorUseRule=UseColor_Foreground)
    InStyle_HoveredForeground:Slate Color Structure=(SpecifiedColor=(R=1,G=0,B=1,A=1),ColorUseRule=UseColor_Foreground)
    InStyle_PressedForeground:Slate Color Structure=(SpecifiedColor=(R=1,G=0,B=1,A=1),ColorUseRule=UseColor_Foreground)
    InStyle_DisabledForeground:Slate Color Structure=(SpecifiedColor=(R=1,G=0,B=1,A=1),ColorUseRule=UseColor_Foreground)
    InStyle_NormalPadding:Margin Structure=(Left=0,Top=0,Right=0,Bottom=0)
    InStyle_PressedPadding:Margin Structure=(Left=0,Top=0,Right=0,Bottom=0)
    InStyle_PressedSlateSound:Slate Sound Structure=(ResourceObject=None)
    InStyle_ClickedSlateSound:Slate Sound Structure=(ResourceObject=None)
    InStyle_HoveredSlateSound:Slate Sound Structure=(ResourceObject=None)
   ) -> (then)
  ]}
 OnClickedButtonEventDispatcher {nodes [OnClickedButtonEventDispatcher!() -> (then, Color:Slate Color Structure) {is=title K2Node_FunctionEntry}]}
 OnRightClickedButtonEventDispatcher {nodes [OnRightClickedButtonEventDispatcher() {is=title K2Node_FunctionEntry}]}
 ThumbnailButtonClicked {nodes [
   ThumbnailButtonClicked() {is=title K2Node_FunctionEntry}
   n1 = Get PaletteButton()
   > Return Node!(execute, Normal Tint:Slate Color Structure=Break Button Style(ButtonStyle=Get WidgetStyle(self=@n1)))
  ]}
 ThumbnailButtonRightClicked {nodes [
   ThumbnailButtonRightClicked() {is=title K2Node_FunctionEntry}
   n1 = Get PaletteButton()
   > Return Node(Normal Tint=Break Button Style(ButtonStyle=Get WidgetStyle(self=@n1)))
  ]}
]
variables [
 Color {type=LinearColor, default=<rgba 0 0 0 1>}
 OnClickedButtonEventDispatcher {default=()}
 OnRightClickedButtonEventDispatcher {default=()}
]
widgets {
 info {rootWidgetClass=&VerticalBox, widgetCount=2}
 widgets [
  VerticalBox_37 {widgetClassPath=&VerticalBox, widget=VerticalBox_37}
  PaletteButton {
   widgetClassPath=&EditorUtilityButton
   parent=VerticalBox_37
   widget=PaletteButton
   properties {widgetStyle {
     clickedSlateSound {resourceObject=null}
     disabled {
      bIsDynamicallyLoaded=false
      drawAs=RoundedBox
      imageType=NoImage
      mirroring=NoMirror
      resourceName=None
      resourceObject=null
      tiling=NoTile
      imageSize=<0 0>
      margin=<ltrb 0 0 0 0>
      outlineSettings {
       bUseBrushTransparency=false
       roundingType=FixedRadius
       width=1
       color {colorUseRule=UseColor_ColorTable, specifiedColor=<rgba 0 0 0 0>}
       cornerRadii=<4 4 4 4>
      }
      tintColor {colorUseRule=UseColor_Specified, specifiedColor=<rgba 0.0395 0.0395 0.0395 1>}
      uVRegion {bIsValid=false, max=<0 0>, min=<0 0>}
     }
     disabledForeground {colorUseRule=UseColor_ColorTable, specifiedColor=<rgba 0 0 0 0>}
     hovered {
      bIsDynamicallyLoaded=false
      drawAs=RoundedBox
      imageType=NoImage
      mirroring=NoMirror
      resourceName=None
      resourceObject=null
      tiling=NoTile
      imageSize=<32 32>
      margin=<ltrb 0 0 0 0>
      outlineSettings {
       bUseBrushTransparency=false
       roundingType=FixedRadius
       width=1
       color {colorUseRule=UseColor_ColorTable, specifiedColor=<rgba 0 0 0 0>}
       cornerRadii=<4 4 4 4>
      }
      tintColor {colorUseRule=UseColor_Specified, specifiedColor=<rgba 0 0.3674 0.5751 1>}
      uVRegion {bIsValid=false, max=<0 0>, min=<0 0>}
     }
     hoveredForeground {colorUseRule=UseColor_ColorTable, specifiedColor=<rgba 0 0 0 0>}
     hoveredSlateSound {resourceObject=null}
     normal {
      bIsDynamicallyLoaded=false
      drawAs=RoundedBox
      imageType=NoImage
      mirroring=NoMirror
      resourceName=None
      resourceObject=null
      tiling=NoTile
      imageSize=<32 32>
      margin=<ltrb 0 0 0 0>
      outlineSettings {
       bUseBrushTransparency=false
       roundingType=FixedRadius
       width=1
       color {colorUseRule=UseColor_ColorTable, specifiedColor=<rgba 0 0 0 0>}
       cornerRadii=<4 4 4 4>
      }
      tintColor {colorUseRule=UseColor_Specified, specifiedColor=<rgba 0.0395 0.0395 0.0395 1>}
      uVRegion {bIsValid=false, max=<0 0>, min=<0 0>}
     }
     normalForeground {colorUseRule=UseColor_ColorTable, specifiedColor=<rgba 0 0 0 0>}
     normalPadding=<ltrb 0 0 0 0>
     pressed {
      bIsDynamicallyLoaded=false
      drawAs=RoundedBox
      imageType=NoImage
      mirroring=NoMirror
      resourceName=None
      resourceObject=null
      tiling=NoTile
      imageSize=<32 32>
      margin=<ltrb 0 0 0 0>
      outlineSettings {
       bUseBrushTransparency=false
       roundingType=FixedRadius
       width=1
       color {colorUseRule=UseColor_ColorTable, specifiedColor=<rgba 0 0 0 0>}
       cornerRadii=<4 4 4 4>
      }
      tintColor {colorUseRule=UseColor_Specified, specifiedColor=<rgba 0.0284 0.0284 0.0284 1>}
      uVRegion {bIsValid=false, max=<0 0>, min=<0 0>}
     }
     pressedForeground {colorUseRule=UseColor_ColorTable, specifiedColor=<rgba 0 0 0 0>}
     pressedPadding=<ltrb 0 0 0 0>
     pressedSlateSound {resourceObject=null}
    }}
   slot {class=VerticalBoxSlot, properties {size {sizeRule=Fill, value=1}}}
  }
 ]
}
