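# Editor utility widget for the AnimationPicker tool that shows a colour palette grid for picking a colour or cancelling.
# On Construct it creates an EUW_PaletteButton per entry in the preset Colors array, wrapping rows by GridCollumsNumber using divide and modulo.
# After the loop it moves CancelButton into the grid's extra top-row column.
# Colour and cancel clicks are re-broadcast through the widget's own dispatchers.
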EditorUtilityWidgetBlueprint /AnimInEdTools/AnimationPicker/EUW_Palette
graphs [
 EventGraph {nodes [
   "On Clicked (CancelButton)"() {is=title K2Node_ComponentBoundEvent}
   > Call Palette Button Cancel Clicked Dispatcher!(execute:Exec, self:EUW Palette Object Reference) -> (then:Exec) {is=title K2Node_CallDelegate}
   n1 = PaletteButtonClicked!() -> (OutputDelegate, then, Color:Slate Color Structure) {is=title K2Node_CustomEvent}
   > Call Palette Button Clicked Dispatcher!(execute:Exec, self:EUW Palette Object Reference, Color:Slate Color Structure=@n1.Color) -> (then:Exec) {is=title K2Node_CallDelegate}
   Event Construct()
   > n2 = For Each Loop(Array:Array of Linear Color Structures=Get Colors() -> (Colors:Array of Linear Color Structures)) -> (Array Element:Linear Color Structure)
   AddChildToUniformGrid!(
    execute=@n2.Completed
    self:Uniform Grid Panel Object Reference=Get AssetThumbnailsUniformGrid() -> (AssetThumbnailsUniformGrid:Uniform Grid Panel Object Reference)
    Content:Widget Object Reference=Get CancelButton() -> (CancelButton:Button Object Reference)
    InRow:int=0
    InColumn:int=12
   ) -> (then, ReturnValue:Uniform Grid Slot Object Reference)
   n3 = int + int(A=@n2.Array Index)
   Set AssetCount(execute=@n2.LoopBody, AssetCount:int=@n3) -> (Output_Get:int=0)
   > n4 = Create Widget!(
    execute:Exec
    WorldContextObject:Object Reference
    Class:User Widget Class Reference=&/EUW_PaletteButton.EUW_PaletteButton_C
    OwningPlayer:Player Controller Object Reference
    Color:Linear Color Structure=@n2
   ) -> (then:Exec, ReturnValue:EUW Palette Button Object Reference)
   n5 = int - int(A=@n3)
   n6 = Get GridCollumsNumber() -> (GridCollumsNumber:int=0)
   > AddChildToUniformGrid(self=Get AssetThumbnailsUniformGrid(), Content=@n4, InRow=int / int(A=@n5, B=@n6), InColumn="% (Integer)"(A=@n5, B=@n6) -> (ReturnValue=0))
   > Bind Event to On Clicked Button Event Dispatcher(self:EUW Palette Button Object Reference=@n4, Delegate=@n1) {is=title K2Node_AddDelegate}
  ]}
 PaletteButtonCancelClickedDispatcher {nodes [PaletteButtonCancelClickedDispatcher() {is=title K2Node_FunctionEntry}]}
 PaletteButtonClickedDispatcher {nodes [PaletteButtonClickedDispatcher!() -> (then, Color:Slate Color Structure) {is=title K2Node_FunctionEntry}]}
]
variables [
 AssetCount {type=integer}
 Colors {
  type="array<LinearColor>"
  default [
   <rgba 0.8656 0.4496 0.4496 1>
   <rgba 0.8332 0.6366 0.437 1>
   <rgba 0.8332 0.8233 0.437 1>
   <rgba 0.6563 0.8332 0.437 1>
   <rgba 0.457 0.8332 0.437 1>
   <rgba 0.4244 0.8007 0.5941 1>
   <rgba 0.4244 0.8007 0.7712 1>
   <rgba 0.4244 0.6533 0.8007 1>
   <rgba 0.437 0.4869 0.8332 1>
   <rgba 0.5767 0.437 0.8332 1>
   <rgba 0.7742 0.437 0.8332 1>
   <rgba 0.8332 0.437 0.7054 1>
   <rgba 0.8656 0.21 0.21 1>
   <rgba 0.8332 0.5225 0.2088 1>
   <rgba 0.8332 0.8176 0.2088 1>
   <rgba 0.5536 0.8332 0.2088 1>
   <rgba 0.2402 0.8332 0.2088 1>
   <rgba 0.2076 0.8007 0.4743 1>
   <rgba 0.2076 0.8007 0.7541 1>
   <rgba 0.2076 0.5678 0.8007 1>
   <rgba 0.2088 0.2873 0.8332 1>
   <rgba 0.4284 0.2088 0.8332 1>
   <rgba 0.74 0.2088 0.8332 1>
   <rgba 0.8332 0.2088 0.6313 1>
   <rgba 0.8506 0.0304 0.0304 1>
   <rgba 0.8189 0.4296 0.0377 1>
   <rgba 0.8189 0.7994 0.0377 1>
   <rgba 0.4686 0.8189 0.0377 1>
   <rgba 0.0769 0.8189 0.0377 1>
   <rgba 0.045 0.7872 0.3782 1>
   <rgba 0.045 0.7872 0.7288 1>
   <rgba 0.045 0.4952 0.7872 1>
   <rgba 0.0377 0.1357 0.8189 1>
   <rgba 0.3121 0.0377 0.8189 1>
   <rgba 0.7021 0.0377 0.8189 1>
   <rgba 0.8189 0.0377 0.5659 1>
   <rgba 0.7234 0.0084 0.0084 1>
   <rgba 0.6977 0.3554 0.0168 1>
   <rgba 0.6977 0.6806 0.0168 1>
   <rgba 0.3897 0.6977 0.0168 1>
   <rgba 0.0506 0.6977 0.0168 1>
   <rgba 0.0251 0.672 0.313 1>
   <rgba 0.0251 0.672 0.6207 1>
   <rgba 0.0251 0.4153 0.672 1>
   <rgba 0.0168 0.1014 0.6977 1>
   <rgba 0.2538 0.0168 0.6977 1>
   <rgba 0.595 0.0168 0.6977 1>
   <rgba 0.6977 0.0168 0.4752 1>
   <rgba 0.5086 0.0304 0.0304 1>
   <rgba 0.4931 0.265 0.0377 1>
   <rgba 0.4931 0.4817 0.0377 1>
   <rgba 0.2878 0.4931 0.0377 1>
   <rgba 0.0604 0.4931 0.0377 1>
   <rgba 0.045 0.4777 0.2382 1>
   <rgba 0.045 0.4777 0.4435 1>
   <rgba 0.045 0.3066 0.4777 1>
   <rgba 0.0377 0.0945 0.4931 1>
   <rgba 0.1968 0.0377 0.4931 1>
   <rgba 0.4247 0.0377 0.4931 1>
   <rgba 0.4931 0.0377 0.3448 1>
   <rgba 0.3109 0.0414 0.0414 1>
   <rgba 0.3254 0.197 0.0687 1>
   <rgba 0.3448 0.3384 0.0881 1>
   <rgba 0.2496 0.3651 0.1084 1>
   <rgba 0.1418 0.3857 0.129 1>
   <rgba 0.1511 0.395 0.2602 1>
   <rgba 0.1696 0.4134 0.3942 1>
   <rgba 0.188 0.3356 0.4319 1>
   <rgba 0.2079 0.24 0.4646 1>
   <rgba 0.3179 0.228 0.4847 1>
   <rgba 0.467 0.2488 0.5055 1>
   <rgba 0.525 0.2683 0.4415 1>
   <rgba 0.0084 0.0084 0.0084 1>
   <rgba 0.0988 0.0988 0.0988 1>
   <rgba 0.1766 0.1766 0.1766 1>
   <rgba 0.2578 0.2578 0.2578 1>
   <rgba 0.3401 0.3401 0.3401 1>
   <rgba 0.4098 0.4098 0.4098 1>
   <rgba 0.4836 0.4836 0.4836 1>
   <rgba 0.5574 0.5574 0.5574 1>
   <rgba 0.6558 0.6558 0.6558 1>
   <rgba 0.7362 0.7362 0.7362 1>
   <rgba 0.8193 0.8193 0.8193 1>
   <rgba 0.8971 0.8971 0.8971 1>
  ]
 }
 GridCollumsNumber {type=integer, default=12}
 PaletteButtonCancelClickedDispatcher {default=()}
 PaletteButtonClickedDispatcher {default=()}
]
widgets {
 info {rootWidgetClass=&VerticalBox, widgetCount=4}
 widgets [
  VerticalBox_0 {widgetClassPath=&VerticalBox, widget=VerticalBox_0}
  ScrollBox_0 {
   widgetClassPath=&ScrollBox
   parent=VerticalBox_0
   widget=ScrollBox_0
   properties {
    widgetBarStyle {
     thickness=8
     draggedThumbImage {
      bIsDynamicallyLoaded=false
      drawAs=RoundedBox
      imageType=NoImage
      mirroring=NoMirror
      resourceName=None
      resourceObject=null
      tiling=NoTile
      imageSize=<0 0>
      margin=<ltrb 0 0 0 0>
      outlineSettings {
       bUseBrushTransparency=false
       roundingType=FixedRadius
       width=0
       color {colorUseRule=UseColor_Specified, specifiedColor=<rgba 0 0 0 0>}
       cornerRadii=<4 4 4 4>
      }
      tintColor {colorUseRule=UseColor_Specified, specifiedColor=<rgba 0.2159 0.2159 0.2159 1>}
      uVRegion {bIsValid=false, max=<0 0>, min=<0 0>}
     }
     horizontalBackgroundImage {
      bIsDynamicallyLoaded=false
      drawAs=NoDrawType
      imageType=NoImage
      mirroring=NoMirror
      resourceName=None
      resourceObject=null
      tiling=NoTile
      imageSize=<0 0>
      margin=<ltrb 0 0 0 0>
      outlineSettings {
       bUseBrushTransparency=false
       roundingType=HalfHeightRadius
       width=0
       color {colorUseRule=UseColor_Specified, specifiedColor=<rgba 0 0 0 0>}
       cornerRadii=<0 0 0 0>
      }
      tintColor {colorUseRule=UseColor_Specified, specifiedColor=<rgba 1 1 1 1>}
      uVRegion {bIsValid=false, max=<0 0>, min=<0 0>}
     }
     horizontalBottomSlotImage {
      bIsDynamicallyLoaded=false
      drawAs=NoDrawType
      imageType=NoImage
      mirroring=NoMirror
      resourceName=None
      resourceObject=null
      tiling=NoTile
      imageSize=<0 0>
      margin=<ltrb 0 0 0 0>
      outlineSettings {
       bUseBrushTransparency=false
       roundingType=HalfHeightRadius
       width=0
       color {colorUseRule=UseColor_Specified, specifiedColor=<rgba 0 0 0 0>}
       cornerRadii=<0 0 0 0>
      }
      tintColor {colorUseRule=UseColor_Specified, specifiedColor=<rgba 1 1 1 1>}
      uVRegion {bIsValid=false, max=<0 0>, min=<0 0>}
     }
     horizontalTopSlotImage {
      bIsDynamicallyLoaded=false
      drawAs=NoDrawType
      imageType=NoImage
      mirroring=NoMirror
      resourceName=None
      resourceObject=null
      tiling=NoTile
      imageSize=<0 0>
      margin=<ltrb 0 0 0 0>
      outlineSettings {
       bUseBrushTransparency=false
       roundingType=HalfHeightRadius
       width=0
       color {colorUseRule=UseColor_Specified, specifiedColor=<rgba 0 0 0 0>}
       cornerRadii=<0 0 0 0>
      }
      tintColor {colorUseRule=UseColor_Specified, specifiedColor=<rgba 1 1 1 1>}
      uVRegion {bIsValid=false, max=<0 0>, min=<0 0>}
     }
     hoveredThumbImage {
      bIsDynamicallyLoaded=false
      drawAs=RoundedBox
      imageType=NoImage
      mirroring=NoMirror
      resourceName=None
      resourceObject=null
      tiling=NoTile
      imageSize=<0 0>
      margin=<ltrb 0 0 0 0>
      outlineSettings {
       bUseBrushTransparency=false
       roundingType=FixedRadius
       width=0
       color {colorUseRule=UseColor_Specified, specifiedColor=<rgba 0 0 0 0>}
       cornerRadii=<4 4 4 4>
      }
      tintColor {colorUseRule=UseColor_Specified, specifiedColor=<rgba 0.2159 0.2159 0.2159 1>}
      uVRegion {bIsValid=false, max=<0 0>, min=<0 0>}
     }
     normalThumbImage {
      bIsDynamicallyLoaded=false
      drawAs=RoundedBox
      imageType=NoImage
      mirroring=NoMirror
      resourceName=None
      resourceObject=null
      tiling=NoTile
      imageSize=<0 0>
      margin=<ltrb 0 0 0 0>
      outlineSettings {
       bUseBrushTransparency=false
       roundingType=FixedRadius
       width=0
       color {colorUseRule=UseColor_Specified, specifiedColor=<rgba 0 0 0 0>}
       cornerRadii=<4 4 4 4>
      }
      tintColor {colorUseRule=UseColor_Specified, specifiedColor=<rgba 0.0953 0.0953 0.0953 1>}
      uVRegion {bIsValid=false, max=<0 0>, min=<0 0>}
     }
     verticalBackgroundImage {
      bIsDynamicallyLoaded=false
      drawAs=NoDrawType
      imageType=NoImage
      mirroring=NoMirror
      resourceName=None
      resourceObject=null
      tiling=NoTile
      imageSize=<0 0>
      margin=<ltrb 0 0 0 0>
      outlineSettings {
       bUseBrushTransparency=false
       roundingType=HalfHeightRadius
       width=0
       color {colorUseRule=UseColor_Specified, specifiedColor=<rgba 0 0 0 0>}
       cornerRadii=<0 0 0 0>
      }
      tintColor {colorUseRule=UseColor_Specified, specifiedColor=<rgba 1 1 1 1>}
      uVRegion {bIsValid=false, max=<0 0>, min=<0 0>}
     }
     verticalBottomSlotImage {
      bIsDynamicallyLoaded=false
      drawAs=NoDrawType
      imageType=NoImage
      mirroring=NoMirror
      resourceName=None
      resourceObject=null
      tiling=NoTile
      imageSize=<0 0>
      margin=<ltrb 0 0 0 0>
      outlineSettings {
       bUseBrushTransparency=false
       roundingType=HalfHeightRadius
       width=0
       color {colorUseRule=UseColor_Specified, specifiedColor=<rgba 0 0 0 0>}
       cornerRadii=<0 0 0 0>
      }
      tintColor {colorUseRule=UseColor_Specified, specifiedColor=<rgba 1 1 1 1>}
      uVRegion {bIsValid=false, max=<0 0>, min=<0 0>}
     }
     verticalTopSlotImage {
      bIsDynamicallyLoaded=false
      drawAs=NoDrawType
      imageType=NoImage
      mirroring=NoMirror
      resourceName=None
      resourceObject=null
      tiling=NoTile
      imageSize=<0 0>
      margin=<ltrb 0 0 0 0>
      outlineSettings {
       bUseBrushTransparency=false
       roundingType=HalfHeightRadius
       width=0
       color {colorUseRule=UseColor_Specified, specifiedColor=<rgba 0 0 0 0>}
       cornerRadii=<0 0 0 0>
      }
      tintColor {colorUseRule=UseColor_Specified, specifiedColor=<rgba 1 1 1 1>}
      uVRegion {bIsValid=false, max=<0 0>, min=<0 0>}
     }
    }
    widgetStyle {
     barThickness=8
     bottomShadowBrush {
      bIsDynamicallyLoaded=false
      drawAs=Box
      imageType=FullColor
      mirroring=NoMirror
      resourceName=../../../Engine/Content/Slate/Common/ScrollBoxShadowBottom.png
      resourceObject=null
      tiling=NoTile
      imageSize=<16 8>
      margin=<ltrb 0.5 0 0.5 1>
      outlineSettings {
       bUseBrushTransparency=false
       roundingType=HalfHeightRadius
       width=0
       color {colorUseRule=UseColor_Specified, specifiedColor=<rgba 0 0 0 0>}
       cornerRadii=<0 0 0 0>
      }
      tintColor {colorUseRule=UseColor_Specified, specifiedColor=<rgba 1 1 1 1>}
      uVRegion {bIsValid=false, max=<0 0>, min=<0 0>}
     }
     horizontalScrolledContentPadding=<ltrb 0 0 1 0>
     leftShadowBrush {
      bIsDynamicallyLoaded=false
      drawAs=Box
      imageType=FullColor
      mirroring=NoMirror
      resourceName=../../../Engine/Content/Slate/Common/ScrollBoxShadowLeft.png
      resourceObject=null
      tiling=NoTile
      imageSize=<8 16>
      margin=<ltrb 1 0.5 0 0.5>
      outlineSettings {
       bUseBrushTransparency=false
       roundingType=HalfHeightRadius
       width=0
       color {colorUseRule=UseColor_Specified, specifiedColor=<rgba 0 0 0 0>}
       cornerRadii=<0 0 0 0>
      }
      tintColor {colorUseRule=UseColor_Specified, specifiedColor=<rgba 1 1 1 1>}
      uVRegion {bIsValid=false, max=<0 0>, min=<0 0>}
     }
     rightShadowBrush {
      bIsDynamicallyLoaded=false
      drawAs=Box
      imageType=FullColor
      mirroring=NoMirror
      resourceName=../../../Engine/Content/Slate/Common/ScrollBoxShadowRight.png
      resourceObject=null
      tiling=NoTile
      imageSize=<8 16>
      margin=<ltrb 0 0.5 1 0.5>
      outlineSettings {
       bUseBrushTransparency=false
       roundingType=HalfHeightRadius
       width=0
       color {colorUseRule=UseColor_Specified, specifiedColor=<rgba 0 0 0 0>}
       cornerRadii=<0 0 0 0>
      }
      tintColor {colorUseRule=UseColor_Specified, specifiedColor=<rgba 1 1 1 1>}
      uVRegion {bIsValid=false, max=<0 0>, min=<0 0>}
     }
     topShadowBrush {
      bIsDynamicallyLoaded=false
      drawAs=Box
      imageType=FullColor
      mirroring=NoMirror
      resourceName=../../../Engine/Content/Slate/Common/ScrollBoxShadowTop.png
      resourceObject=null
      tiling=NoTile
      imageSize=<16 8>
      margin=<ltrb 0.5 1 0.5 0>
      outlineSettings {
       bUseBrushTransparency=false
       roundingType=HalfHeightRadius
       width=0
       color {colorUseRule=UseColor_Specified, specifiedColor=<rgba 0 0 0 0>}
       cornerRadii=<0 0 0 0>
      }
      tintColor {colorUseRule=UseColor_Specified, specifiedColor=<rgba 1 1 1 1>}
      uVRegion {bIsValid=false, max=<0 0>, min=<0 0>}
     }
     verticalScrolledContentPadding=<ltrb 0 0 0 1>
    }
   }
   slot {class=VerticalBoxSlot, properties {size {sizeRule=Fill, value=1}}}
  }
  AssetThumbnailsUniformGrid {
   widgetClassPath=/Script/UMG.UniformGridPanel
   bIsVariable=true
   parent=ScrollBox_0
   widget=AssetThumbnailsUniformGrid
   slot {class=ScrollBoxSlot, properties {horizontalAlignment=HAlign_Center, size {sizeRule=Fill, value=1}}}
  }
  CancelButton {
   widgetClassPath=&Button
   parent=ScrollBox_0
   widget=CancelButton
   properties {
    toolTipText=Cancel
    widgetStyle {
     clickedSlateSound {resourceObject=null}
     disabled {
      bIsDynamicallyLoaded=false
      drawAs=RoundedBox
      imageType=NoImage
      mirroring=NoMirror
      resourceName=None
      resourceObject=null
      tiling=NoTile
      imageSize=<0 0>
      margin=<ltrb 0 0 0 0>
      outlineSettings {
       bUseBrushTransparency=false
       roundingType=FixedRadius
       width=1
       color {colorUseRule=UseColor_ColorTable, specifiedColor=<rgba 0 0 0 0>}
       cornerRadii=<4 4 4 4>
      }
      tintColor {colorUseRule=UseColor_Specified, specifiedColor=<rgba 0.0395 0.0395 0.0395 1>}
      uVRegion {bIsValid=false, max=<0 0>, min=<0 0>}
     }
     disabledForeground {colorUseRule=UseColor_ColorTable, specifiedColor=<rgba 0 0 0 0>}
     hovered {
      bIsDynamicallyLoaded=false
      drawAs=Image
      imageType=FullColor
      mirroring=NoMirror
      resourceName=None
      resourceObject=&Cancel_Normal
      tiling=NoTile
      imageSize=<30 30>
      margin=<ltrb 0 0 0 0>
      outlineSettings {
       bUseBrushTransparency=false
       roundingType=HalfHeightRadius
       width=0
       color {colorUseRule=UseColor_ColorTable, specifiedColor=<rgba 0 0 0 0>}
       cornerRadii=<10 10 10 10>
      }
      tintColor {colorUseRule=UseColor_Specified, specifiedColor=<rgba 0.702 0.5998 0.6112 1>}
      uVRegion {bIsValid=false, max=<0 0>, min=<0 0>}
     }
     hoveredForeground {colorUseRule=UseColor_ColorTable, specifiedColor=<rgba 0 0 0 0>}
     hoveredSlateSound {resourceObject=null}
     normal {
      bIsDynamicallyLoaded=false
      drawAs=Image
      imageType=FullColor
      mirroring=NoMirror
      resourceName=None
      resourceObject=&Cancel_Normal
      tiling=NoTile
      imageSize=<30 30>
      margin=<ltrb 0 0 0 0>
      outlineSettings {
       bUseBrushTransparency=false
       roundingType=HalfHeightRadius
       width=0
       color {colorUseRule=UseColor_ColorTable, specifiedColor=<rgba 0 0 0 0>}
       cornerRadii=<10 10 10 10>
      }
      tintColor {colorUseRule=UseColor_Specified, specifiedColor=<rgba 0.702 0.1622 0.2228 1>}
      uVRegion {bIsValid=false, max=<0 0>, min=<0 0>}
     }
     normalForeground {colorUseRule=UseColor_ColorTable, specifiedColor=<rgba 0 0 0 0>}
     normalPadding=<ltrb 12 1.5 12 1.5>
     pressed {
      bIsDynamicallyLoaded=false
      drawAs=Image
      imageType=FullColor
      mirroring=NoMirror
      resourceName=None
      resourceObject=&Cancel_Normal
      tiling=NoTile
      imageSize=<30 30>
      margin=<ltrb 0 0 0 0>
      outlineSettings {
       bUseBrushTransparency=false
       roundingType=HalfHeightRadius
       width=0
       color {colorUseRule=UseColor_ColorTable, specifiedColor=<rgba 0 0 0 0>}
       cornerRadii=<10 10 10 10>
      }
      tintColor {colorUseRule=UseColor_Specified, specifiedColor=<rgba 1 1 1 1>}
      uVRegion {bIsValid=false, max=<0 0>, min=<0 0>}
     }
     pressedForeground {colorUseRule=UseColor_ColorTable, specifiedColor=<rgba 0 0 0 0>}
     pressedPadding=<ltrb 12 2.5 12 0.5>
     pressedSlateSound {resourceObject=null}
    }
   }
   slot {class=ScrollBoxSlot, properties {padding=<ltrb 2 0 2 0>}}
  }
 ]
}
