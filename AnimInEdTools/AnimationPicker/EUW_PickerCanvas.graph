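# EUW_PickerCanvas is an Editor Utility Widget that, by its name and folder, provides a canvas for the AnimInEdTools animation picker.
# Its widget tree holds only an empty root CanvasPanel, exposed as a variable.
# Variables store a dirty flag, a load type string, and saved positions and a transform for a mover canvas.
# The PreConstruct, Construct and Tick events have no logic attached.
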EditorUtilityWidgetBlueprint /AnimInEdTools/AnimationPicker/EUW_PickerCanvas
graphs [EventGraph {nodes [
   Event PreConstruct()
   Event Construct()
   Event Tick()
  ]}]
variables [
 Dirty {type=boolean}
 Load Type {type=string, default=""}
 Mover Canvas Saved Positions {default {}}
 Mover Canvas Transform {type=WidgetTransform, default {angle=0, scale=<1 1>, shear=<0 0>, translation=<0 0>}}
]
widgets {
 info {rootWidgetClass=&CanvasPanel, widgetCount=1}
 widgets [CanvasPanel {widgetClassPath=&CanvasPanel, bIsVariable=true, widget=CanvasPanel}]
}
