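# Base Editor Utility Widget for per-asset animation pickers, storing the Control Rig binding, creature and parent picker passed on load.
# Helpers gather selectable elements from an overridable primary panel and bind buttons to rig controls by matching converted widget names.
# IK/FK helpers key bool channels in Sequencer and sync checkboxes, button visibility and tint.
# Variant functions add offset skeletal mesh spawnables playing the sequence's linked animation, optionally with layered rigs, or remove them by name.
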EditorUtilityWidgetBlueprint /AnimInEdTools/AnimationPicker/EUW_Asset_Picker_Base
graphs [
 Add Variants to Sequence {nodes [
   Add Variants to Sequence!() -> (
    then
    Skeletal Meshes:Array of Skeletal Mesh Object References
    Location Offset:Vector=~
    Layered Control Rig:Control Rig Asset Strong Reference Structure
   ) {is=title K2Node_FunctionEntry}
   > n1 = Sequence()
   > n2 = Sequence!(execute) -> (then_0, then_1, then_2:Exec)
   n3 = GetCurrentLevelSequence!(execute=@n2.then_1) -> (then, ReturnValue:Level Sequence Object Reference)
   > n4 = For Each Loop(Array:Array of Skeletal Mesh Object References=Get Skeletal Meshes() -> (Skeletal Meshes:Array of Skeletal Mesh Object References)) -> (Array Element:Skeletal Mesh Object Reference)
   > Set Local Current SKM Index(Local Current SKM Index:int=@n4.Array Index) -> (Output_Get:int=0)
   > n5 = AddSpawnableFromInstance!(
    execute
    self:Level Sequence Editor Subsystem Object Reference=Get LevelSequenceEditorSubsystem() -> (ReturnValue:Level Sequence Editor Subsystem Object Reference) {is=title K2Node_GetEditorSubsystem}
    Sequence:Movie Scene Sequence Object Reference=@n3
    ObjectToSpawn:Object Reference=@n4
   ) -> (then, ReturnValue:Movie Scene Binding Proxy Structure)
   > n6 = Sequence!(execute) -> (then_0, then_1)
   n7 = GetTracks!(execute=@n6.then_1, InBinding:"Movie Scene Binding Proxy Structure (by ref)"=@n5) -> (then, ReturnValue:Array of Movie Scene Track Object References)
   > n8 = For Each Loop(Array:Array of Movie Scene Track Object References=@n7) -> (Array Element:Movie Scene Track Object Reference)
   > n9 = Cast To MovieScene3DTransformTrack!(execute, Object=@n8) -> (then, CastFailed, AsMovie Scene 3DTransform Track:Movie Scene 3DTransform Track Object Reference) {is=title K2Node_DynamicCast}
   > n10 = GetSections!(execute, Track:Movie Scene Track Object Reference=@n9) -> (then, ReturnValue:Array of Movie Scene Section Object References)
   > n11 = For Each Loop(Array:Array of Movie Scene Section Object References=@n10) -> (Array Element:Movie Scene Section Object Reference)
   > n12 = Cast To MovieScene3DTransformSection!(execute, Object=@n11) -> (then, CastFailed, AsMovie Scene 3DTransform Section:Movie Scene 3DTransform Section Object Reference) {is=title K2Node_DynamicCast}
   > n13 = GetAllChannels!(execute, Section:Movie Scene Section Object Reference=@n12) -> (then, ReturnValue:Array of Movie Scene Scripting Channel Object References)
   > n14 = For Each Loop(Array:Array of Movie Scene Scripting Channel Object References=@n13) -> (Array Element:Movie Scene Scripting Channel Object Reference)
   > n15 = Cast To MovieSceneScriptingDoubleChannel!(execute, Object=@n14) -> (then, CastFailed, AsMovie Scene Scripting Double Channel:Movie Scene Scripting Double Channel Object Reference) {is=title K2Node_DynamicCast}
   > n16 = Branch(Condition="Equal (Integer)"(A=@n14.Array Index))
   n17 = float + float(A="To Float (Integer)"(InInt=Get Local Current SKM Index() -> (Local Current SKM Index:int=0)), B=1)
   n18 = Get!() -> (Location Offset_X:double, Location Offset_Y:double, Location Offset_Z:double)
   > "Set Default (Double)"!(
    execute
    self:Movie Scene Scripting Double Channel Object Reference=@n15
    InDefaultValue:double=float * float(A=@n17, B=@n18) {is=title K2Node_PromotableOperator}
   ) -> (then) {is=title K2Node_CallFunction}
   n19 = Branch(execute=@n16.else, Condition="Equal (Integer)"(A=@n14.Array Index, B=1))
   > "Set Default (Double)"(self=@n15, InDefaultValue=float * float(A=@n18.Location Offset_Y, B=@n17))
   Branch(execute=@n19.else, Condition="Equal (Integer)"(A=@n14.Array Index, B=2))
   > "Set Default (Double)"(self=@n15, InDefaultValue=float * float(A=@n18.Location Offset_Z, B=@n17))
   n20 = AddTrack!(
    execute=@n6.then_0
    InBinding:"Movie Scene Binding Proxy Structure (by ref)"=@n5
    TrackType:Movie Scene Track Class Reference=/Script/MovieSceneTracks.MovieSceneSkeletalAnimationTrack
   ) -> (then, ReturnValue:Movie Scene Track Object Reference)
   > n21 = AddSection!(execute, Track:Movie Scene Track Object Reference=@n20) -> (then, ReturnValue:Movie Scene Section Object Reference)
   > n22 = Cast To MovieSceneSkeletalAnimationSection!(execute, Object=@n21) -> (then, CastFailed, AsMovie Scene Skeletal Animation Section:Movie Scene Skeletal Animation Section Object Reference) {is=title K2Node_DynamicCast}
   n23 = Get Local Anim Sequence() -> (Local Anim Sequence:Anim Sequence Object Reference)
   > n24 = Set Params!(
    execute
    Params:Movie Scene Skeletal Animation Params Structure=Make Movie Scene Skeletal Animation Params!(
     Animation:Anim Sequence Base Object Reference=@n23
     FirstLoopStartFrameOffset:Frame Number Structure=(Value=0)
     StartFrameOffset:Frame Number Structure=(Value=0)
     EndFrameOffset:Frame Number Structure=(Value=0)
     PlayRate:Movie Scene Time Warp Variant Structure="EMovieSceneTimeWarpType::FixedPlayRate(PlayRate=1)"
     bReverse:bool=False
     SlotName:Name=DefaultSlot
     MirrorDataTable:Mirror Data Table Object Reference
     bSkipAnimNotifiers:bool=False
     bForceCustomMode:bool=False
     SwapRootBone:ESwapRootBone Enum=SwapRootBone_None
     bLinearPlaybackWhenScaled:bool=False
    ) -> (MovieSceneSkeletalAnimationParams:Movie Scene Skeletal Animation Params Structure) {is=title K2Node_MakeStruct}
    self:Movie Scene Skeletal Animation Section Object Reference=@n22
   ) -> (then, Output_Get:Movie Scene Skeletal Animation Params Structure)
   > SetEditorProperty(Object=@n22, PropertyName=Params, PropertyValue:Movie Scene Skeletal Animation Params Structure=@n24)
   > SetRange!(
    execute
    Section:Movie Scene Section Object Reference=@n22
    StartFrame:int=0
    EndFrame:int=int + int(A=GetNumFrames!(AnimationSequenceBase:Anim Sequence Base Object Reference=@n23) -> (NumFrames:int=0))
   ) -> (then)
   n25 = Get Layered Control Rig() -> (Layered Control Rig:Control Rig Asset Strong Reference Structure)
   > Branch(Condition=Is Valid Referfence!(InRef:"Control Rig Asset Strong Reference Structure (by ref)"=@n25) -> (ReturnValue:bool))
   > n26 = GetCurrentLevelSequence()
   > n27 = GetEditorWorld(self=Get UnrealEditorSubsystem())
   > FindOrCreateControlRigTrackFromReference!(
    execute
    World:World Object Reference=@n27
    LevelSequence:Level Sequence Object Reference=@n26
    InRef:"Control Rig Asset Strong Reference Structure (by ref)"=@n25
    InBinding:"Movie Scene Binding Proxy Structure (by ref)"=@n5
    bIsLayeredControlRig:bool=true
   ) -> (then, ReturnValue:Movie Scene Track Object Reference)
   n28 = GetCurrentLevelSequence(execute=@n2.then_0)
   > n29 = GetAnimSequenceLinkFromLevelSequence!(execute, InLevelSequence:Level Sequence Object Reference=@n28) -> (then, ReturnValue:Level Sequence Anim Sequence Link Object Reference)
   > n30 = LoadAsset(AssetPath="Get (a copy)"!(
     Array:Array of Level Sequence Anim Sequence Link Item Structures=Get AnimSequenceLinks!(self:Level Sequence Anim Sequence Link Object Reference=@n29) -> (AnimSequenceLinks:Array of Level Sequence Anim Sequence Link Item Structures)
     Dimension 1
    ) -> (
     Output_SkelTrackGuid:Guid Structure
     Output_PathToAnimSequence_PathString:String
     Output_bAutoBake:bool=~
     Output_bExportTransforms:bool=~
     Output_bExportMorphTargets:bool=~
     Output_bExportAttributeCurves:bool=~
     Output_bExportMaterialCurves:bool=~
     Output_bSkipCurvesWithZeroValue:bool=~
     Output_bSetRetargetSourceAsset:bool=~
     Output_Interpolation:EAnimInterpolationType Enum
     Output_CurveInterpolation:ERichCurveInterpMode Enum
     Output_bRecordInWorldSpace:bool=~
     Output_bEvaluateAllSkeletalMeshComponents:bool=~
     Output_IncludeAnimationNames:Array of Strings
     Output_ExcludeAnimationNames:Array of Strings
     Output_bRemoveExcludedCurves:bool=~
     Output_WarmUpFrames:Frame Number Structure
     Output_DelayBeforeStart:Frame Number Structure
     Output_bUseCustomTimeRange:bool=~
     Output_CustomStartFrame:Frame Number Structure
     Output_CustomEndFrame:Frame Number Structure
     Output_CustomDisplayRate:Frame Rate Structure
     Output_bUseCustomFrameRate:bool=~
     Output_CustomFrameRate:Frame Rate Structure
    ).Output_PathToAnimSequence_PathString)
   > n31 = Cast To AnimSequence!(execute, Object=@n30) -> (then, CastFailed, AsAnim Sequence:Anim Sequence Object Reference)
   > Set Local Anim Sequence(Local Anim Sequence:Anim Sequence Object Reference=@n31) -> (Output_Get:Anim Sequence Object Reference)
   Return Node(execute=@n1.then_1)
  ]}
 Convert to EUW naming {nodes [
   n1 = Convert to EUW naming!() -> (then, In name:String) {is=title K2Node_FunctionEntry}
   > Return Node!(execute, ReturnValue:String=Replace(SourceString=Replace(SourceString=@n1, From=/, To=_), From=" ", To=-))
  ]}
 Convert to MR naming {nodes [
   n1 = Convert to MR naming!() -> (then, In name:String) {is=title K2Node_FunctionEntry}
   n2 = Replace(SourceString=@n1, From=_, To=/)
   > Return Node!(execute, Out name:Name=String To Name(InString=Replace(SourceString=@n2, From=-, To=" ")))
  ]}
 EventGraph {nodes [
   n1 = Event Widget has loaded!() -> (
    OutputDelegate
    then
    Control Rig Proxy Binding:Control Rig Sequencer Binding Proxy Structure
    Animatable Creature:Animatable Creature Object Reference
    Anim Picker EUW:Animation Picker Object Reference
   ) {is=title K2Node_Event}
   > Set Control Rig Proxy Binding(Control Rig Proxy Binding:Control Rig Sequencer Binding Proxy Structure=@n1.Control Rig Proxy Binding) -> (Output_Get:Control Rig Sequencer Binding Proxy Structure)
   > Set Animatable Creature(Animatable Creature:Animatable Creature Object Reference=@n1.Animatable Creature) -> (Output_Get:Animatable Creature Object Reference)
   > n2 = Set Anim Picker EUW(Anim Picker EUW:Animation Picker Object Reference=@n1.Anim Picker EUW) -> (Output_Get:Animation Picker Object Reference)
   > Update All Picker Elements Array(self:Animation Picker Object Reference=@n2)
  ]}
 Get Primary Panel {nodes [
   Get Primary Panel() {is=title K2Node_FunctionEntry}
   > Return Node!(execute, Panel:Panel Widget Object Reference)
  ]}
 Get all Selectable Elements {nodes [
   Get all Selectable Elements() {is=title K2Node_FunctionEntry}
   > Clear(TargetArray:Array of EUW Picker Element Object References=Get Local Elem() -> (Local Elem:Array of EUW Picker Element Object References))
   > n1 = Get Primary Panel!(execute, self) -> (then, Panel:Panel Widget Object Reference) {is=title K2Node_CallFunction}
   > n2 = Is Valid(InputObject=@n1)
   LogString(execute=@n2.Is Not Valid, InString=No Primary Panel found)
   > Return Node!(execute, All Elements:Array of EUW Picker Element Object References)
   n3 = For Each Loop(Exec=@n2.Is Valid, Array:Array of Widget Object References=GetAllChildren(self=@n1)) -> (Array Element:Widget Object Reference)
   Return Node(execute=@n3.Completed, All Elements=Get Local Elem())
   > n4 = Set Current Widget(Current Widget:Widget Object Reference=@n3) -> (Output_Get:Widget Object Reference)
   n5 = Branch(
    execute=[@n4.then, @n16.else, @n9.else]
    Condition=Does Object Implement Interface!(
     TestObject:Object Reference=Get Current Widget() -> (Current Widget:Widget Object Reference)
     Interface:Interface Interface=&/BPI_AnimationPickerWidget.BPI_AnimationPickerWidget_C
    ) -> (ReturnValue:bool)
   )
   > n6 = Get Primary Panel!(execute:Exec, self:Object Reference=Get Current Widget()) -> (then:Exec, Panel:Panel Widget Object Reference) {is=title K2Node_Message}
   > n7 = For Each Loop(Array=GetAllChildren(self=@n6))
   > n8 = Set Current Widget(Current Widget=@n7)
   > n9 = Branch(Condition=Does Object Implement Interface(TestObject=@n8, Interface=/AnimInEdTools/AnimationPicker/PickerElements/BPI_SelectablePickerElements.BPI_SelectablePickerElements_C))
   > n10 = Cast To EUW_Picker_Element!(execute, Object=Get Current Widget()) -> (then, CastFailed, AsEUW Picker Element:EUW Picker Element Object Reference) {is=title K2Node_DynamicCast}
   n11 = Get Local Elem()
   > n12 = Branch(Condition=Contains Item(TargetArray:Array of EUW Picker Element Object References=@n11, ItemToFind:"EUW Picker Element Object Reference (by ref)"=@n10))
   Add Unique(
    execute=@n12.else
    TargetArray:Array of EUW Picker Element Object References=@n11
    NewItem:"EUW Picker Element Object Reference (by ref)"=@n10
   )
   n13 = Cast To PanelWidget!(execute=@n5.else, Object=Get Current Widget()) -> (then, CastFailed, AsPanel Widget:Panel Widget Object Reference) {is=title K2Node_DynamicCast}
   > n14 = For Each Loop(Array=GetAllChildren(self=@n13))
   > n15 = Set Current Widget(Current Widget=@n14)
   > n16 = Branch(Condition=Does Object Implement Interface(TestObject=@n15, Interface=&BPI_SelectablePickerElements_C))
   > n17 = Cast To EUW_Picker_Element(Object=Get Current Widget())
   n18 = Get Local Elem()
   > n19 = Branch(Condition=Contains Item(TargetArray=@n18, ItemToFind=@n17))
   Add Unique(execute=@n19.else, TargetArray=@n18, NewItem=@n17)
   n20 = Cast To EUW_Picker_Element(execute=@n13.CastFailed, Object=Get Current Widget())
   > Add Unique(TargetArray=Get Local Elem(), NewItem=@n20)
  ]}
 Print Button Names {nodes [
   n1 = Print Button Names!() -> (then, Control Rig:Control Rig Object Reference) {is=title K2Node_FunctionEntry}
   > n2 = For Each Loop!(
    Exec
    Array:Array of Rig Element Key Structures=Get Controls!(
     self:Rig Hierarchy Object Reference=GetHierarchy!(self:Control Rig Object Reference=@n1) -> (ReturnValue:Rig Hierarchy Object Reference)
     bTraverse:bool=true
    ) -> (ReturnValue:Array of Rig Element Key Structures) {is=title K2Node_CallFunction}
   ) -> (LoopBody, Array Element_Type:ERigElementType Enum, Array Element_Name:Name, Array Index, Completed)
   > PrintString(InString=Convert to EUW naming!(In name:String="To String (Name)"(InName=@n2.Array Element_Name)) -> (ReturnValue:String) {is=title K2Node_CallFunction})
  ]}
 Remove Variants from Sequence {nodes [
   Remove Variants from Sequence!() -> (then, Skeletal Meshes:Array of Skeletal Mesh Object References) {is=title K2Node_FunctionEntry}
   > n1 = Sequence()
   n2 = GetAllLevelActors(execute=@n1.then_1, self=Get EditorActorSubsystem())
   > n3 = For Each Loop!(Exec, Array:Array of Actor Object References=@n2) -> (LoopBody, Array Element:Actor Object Reference, Array Index, Completed)
   > n4 = For Each Loop(Array:Array of Object References=Get Skeletal Meshes()) -> (Array Element:Object Reference)
   > Branch(Condition=Contains(SearchIn=GetDisplayName(Object=@n3), Substring=GetDisplayName(Object=@n4)))
   > Destroy Actor(self=@n3)
   n5 = GetCurrentLevelSequence(execute=@n1.then_0)
   > n6 = GetBindings!(execute, Sequence:Movie Scene Sequence Object Reference=@n5) -> (then, ReturnValue:Array of Movie Scene Binding Proxy Structures)
   > n7 = For Each Loop(Array:Array of Movie Scene Binding Proxy Structures=@n6) -> (Array Element:Movie Scene Binding Proxy Structure)
   > n8 = GetDisplayName!(execute:Exec, InBinding:"Movie Scene Binding Proxy Structure (by ref)"=@n7) -> (then:Exec, ReturnValue:Text)
   > n9 = For Each Loop(Array:Array of Object References=Get Skeletal Meshes()) -> (Array Element:Object Reference)
   > Branch(Condition=Contains(SearchIn="To String (Text)"(InText=@n8), Substring=GetDisplayName!(Object:Object Reference=@n9) -> (ReturnValue:String)))
   > Remove!(execute, InBinding:"Movie Scene Binding Proxy Structure (by ref)"=@n7) -> (then)
  ]}
 Select Multiple Buttons {nodes [
   n1 = Select Multiple Buttons!() -> (then, Button:EUW Picker Button Object Reference, Buttons to Select:Array of EUW Picker Button Object References) {is=title K2Node_FunctionEntry}
   > Clear!(
    execute
    TargetMap:Map of Control Rig Object References to Animation Picker String Array Struct Structures=Get Control Bindings!(self:EUW Picker Element Object Reference=@n1) -> (Control Bindings:Map of Control Rig Object References to Animation Picker String Array Struct Structures)
   ) -> (then) {is=title K2Node_CallFunction}
   > Clear(TargetArray:Array of Names=Get Control Names() -> (Control Names:Array of Names)) {is=title K2Node_CallArrayFunction}
   > n2 = For Each Loop(Array:Array of EUW Picker Button Object References=@n1.Buttons to Select) -> (Array Element:EUW Picker Button Object Reference)
   Set Controls for Primary Control Rig!(execute=@n2.Completed, self:EUW Picker Element Object Reference=@n1, Control Names:Array of Names=Get Control Names()) -> (then) {is=title K2Node_CallFunction}
   > Return Node!(execute)
   n3 = Values!(
    execute=@n2.LoopBody
    TargetMap:Map of Control Rig Object References to Animation Picker String Array Struct Structures=Get Control Bindings(self=@n2)
   ) -> (then, Values:Array of Animation Picker String Array Struct Structures)
   > n4 = For Each Loop!(Exec, Array:Array of Animation Picker String Array Struct Structures=@n3) -> (LoopBody, Array Element_ControlNames_6_#1:Array of Names, Array Index, Completed)
   > n5 = For Each Loop!(Exec, Array:Array of Names=@n4) -> (LoopBody, Array Element:Name, Array Index, Completed)
   > Add Unique(TargetArray:Array of Names=Get Control Names(), NewItem:"Name (by ref)"=@n5)
  ]}
 Select Multiple Buttons Deprecated {nodes [
   n1 = Select Multiple Buttons Deprecated!() -> (then, Buttons:Array of EUW Picker Button Object References, Control Rig:Control Rig Object Reference) {is=title K2Node_FunctionEntry}
   > SelectNothing(self:Editor Actor Subsystem Object Reference=Get EditorActorSubsystem())
   > n2 = GetFocusedLevelSequence!(execute) -> (then, ReturnValue:Level Sequence Object Reference)
   > n3 = GetControlRigs!(execute, LevelSequence:Level Sequence Object Reference=@n2) -> (then, ReturnValue:Array of Control Rig Sequencer Binding Proxy Structures)
   > n4 = For Each Loop!(Exec, Array:Array of Control Rig Sequencer Binding Proxy Structures=@n3) -> (
    LoopBody
    Array Element_Proxy:Movie Scene Binding Proxy Structure
    Array Element_ControlRig:Control Rig Object Reference
    Array Element_Track:Movie Scene Control Rig Parameter Track Object Reference
    Array Index
    Completed
   )
   n5 = For Each Loop!(Exec=@n4.Completed, Array:Array of EUW Picker Button Object References=@n1) -> (LoopBody, Array Element:EUW Picker Button Object Reference, Array Index, Completed)
   > n6 = Values(TargetMap=Get Control Bindings(self=@n5))
   > n7 = For Each Loop(Array:Array of Animation Picker String Array Struct Structures=@n6) -> (Array Element:Animation Picker String Array Struct Structure)
   > n8 = For Each Loop(Array:Array of Names=Break Animation Picker String Array Struct!(Animation_Picker_StringArray_Struct:"Animation Picker String Array Struct Structure (by ref)"=@n7) -> (ControlNames_6_#1:Array of Names) {is=title K2Node_BreakStruct}) -> (Array Element:Name)
   > SelectControl!(execute, self:Control Rig Object Reference=@n1.Control Rig, InControlName:"Name (by ref)"=@n8, bSelect:bool=true, bSetupUndo:bool=false) -> (then)
   ClearControlSelection!(execute=@n4.LoopBody, self:Control Rig Object Reference=@n4.Array Element_ControlRig, bSetupUndo:bool=false) -> (then, ReturnValue:bool)
  ]}
 Set Animation Channel Bool {nodes [
   n1 = Set Animation Channel Bool!() -> (then, ControlRig:Control Rig Object Reference, Channel Name:Name, Value:bool=~) {is=title K2Node_FunctionEntry}
   > n2 = Get Current Local Time!(execute, TimeUnit:EMovieSceneTimeUnit Enum=DisplayRate) -> (
    then
    ReturnValue_Frame_FrameNumber:Frame Number Structure
    ReturnValue_Frame_SubFrame:float=~
    ReturnValue_Time:float=~
    ReturnValue_MarkedFrame:String
    ReturnValue_Timecode:Timecode Structure
    ReturnValue_PositionType:EMovieScenePositionType Enum
    ReturnValue_UpdateMethod:EUpdatePositionMethod Enum
    ReturnValue_bHasJumped:bool=~
   ) {is=title K2Node_CallFunction}
   > n3 = GetFocusedLevelSequence()
   > SetInteraction!(execute, bIsInteracting:bool=true) -> (then)
   > SetLocalControlRigBool!(
    execute
    LevelSequence:Level Sequence Object Reference=@n3
    ControlRig:Control Rig Object Reference=@n1
    ControlName:Name=@n1.Channel Name
    Frame:Frame Number Structure=@n2
    Value:bool=@n1.Value
    TimeUnit:EMovieSceneTimeUnit Enum=DisplayRate
    bSetKey:bool=true
   ) -> (then)
   > SetInteraction(bIsInteracting=false)
  ]}
 Set Buttons Color From Checkbox State {nodes [
   n1 = Set Buttons Color From Checkbox State!() -> (
    then
    CheckBox:Editor Utility Check Box Object Reference
    IK Buttons:Array of EUW Picker Button Object References
    FK Buttons:Array of EUW Picker Button Object References
    Color Mult:double=0.5
    Opacity:double=0.5
   ) {is=title K2Node_FunctionEntry}
   > n2 = Sequence()
   > n3 = Switch on ECheckBoxState!(
    execute
    Selection:ECheckBoxState Enum=Get CheckedState!(self:Check Box Object Reference=@n1) -> (CheckedState:ECheckBoxState Enum=Unchecked)
   ) -> (Unchecked:Exec, Checked:Exec, Undetermined:Exec) {is=title K2Node_SwitchEnum}
   Set bIsChecked(execute=@n3.Checked, bIsChecked:bool=true) -> (Output_Get:bool)
   Set bIsChecked(execute=@n3.Unchecked, bIsChecked=false)
   n4 = For Each Loop(
    Exec=@n2.then_1
    Array:Array of EUW Picker Button Object References=Get FK Buttons() -> (FK Buttons:Array of EUW Picker Button Object References)
   ) -> (Array Element:EUW Picker Button Object Reference)
   n5 = For Each Loop(Exec=@n4.Completed, Array=Get IK Buttons() -> (IK Buttons:Array of EUW Picker Button Object References))
   n6 = Get Color Mult() -> (Color Mult:double)
   n7 = Get bIsChecked() -> (bIsChecked:bool)
   n8 = SelectFloat(A=float / float(A=1, B=@n6), B=@n6, bPickA=@n7)
   n9 = Get!(self:User Widget Object Reference=@n5) -> (ColorAndOpacity_R:float=~, ColorAndOpacity_G:float=~, ColorAndOpacity_B:float=~, ColorAndOpacity_A:float=~)
   n10 = float * float(A=@n9.ColorAndOpacity_B, B=@n8)
   n11 = float * float(A=@n9, B=@n8)
   n12 = float * float(A=@n9.ColorAndOpacity_G, B=@n8)
   > Set ColorAndOpacity!(
    execute
    ColorAndOpacity:Linear Color Structure=MakeColor!(
     R:float="Clamp (Float)"(Value=@n11, Min=0, Max=1)
     G:float="Clamp (Float)"(Value=@n12, Min=0, Max=1)
     B:float="Clamp (Float)"(Value=@n10, Min=0, Max=1)
     A:float=SelectFloat(A=1, B=@n1.Opacity, bPickA=@n7)
    ) -> (ReturnValue:Linear Color Structure)
    self:User Widget Object Reference=@n5
   ) -> (then, Output_Get:Linear Color Structure)
   n13 = NOT Boolean(A=Get bIsChecked())
   n14 = Get Color Mult()
   n15 = SelectFloat(A=float / float(A=1, B=@n14), B=@n14, bPickA=@n13)
   n16 = Get(self=@n4) -> (ColorAndOpacity_R=~, ColorAndOpacity_G=~, ColorAndOpacity_B=~, ColorAndOpacity_A=~)
   n17 = float * float(A=@n16, B=@n15)
   n18 = float * float(A=@n16.ColorAndOpacity_G, B=@n15)
   n19 = float * float(A=@n16.ColorAndOpacity_B, B=@n15)
   Set ColorAndOpacity(
    execute=@n4.LoopBody
    ColorAndOpacity=MakeColor(
     R="Clamp (Float)"(Value=@n17, Min=0, Max=1)
     G="Clamp (Float)"(Value=@n18, Min=0, Max=1)
     B="Clamp (Float)"(Value=@n19, Min=0, Max=1)
     A=SelectFloat(A=1, B=@n1.Opacity, bPickA=@n13)
    )
    self=@n4
   )
  ]}
 Set Buttons Visibility From AnimationChannel {nodes [
   n1 = Set Buttons Visibility From AnimationChannel!() -> (then, ControlRig:Control Rig Object Reference, AnimationChannel:String, Buttons:Array of EUW Picker Button Object References) {is=title K2Node_FunctionEntry}
   > n2 = Get Current Local Time(TimeUnit=DisplayRate) -> (ReturnValue_Frame_SubFrame=~, ReturnValue_Time=~, ReturnValue_bHasJumped=~)
   > n3 = GetFocusedLevelSequence()
   > n4 = GetLocalControlRigBool!(
    execute
    LevelSequence:Level Sequence Object Reference=@n3
    ControlRig:Control Rig Object Reference=@n1
    ControlName:Name=Convert to MR naming!(In name:String=@n1.AnimationChannel) -> (Out name:Name=None) {is=title K2Node_CallFunction}
    Frame:Frame Number Structure=@n2
    TimeUnit:EMovieSceneTimeUnit Enum=DisplayRate
   ) -> (then, ReturnValue:bool)
   > Set Visibility for Array!(
    execute
    self
    " Array":Array of EUW Picker Button Object References=Get Buttons() -> (Buttons:Array of EUW Picker Button Object References)
    bVisibility:bool=@n4
   ) -> (then) {is=title K2Node_CallFunction}
  ]}
 Set Buttons Visibility From Checkbox State {nodes [
   n1 = Set Buttons Visibility From Checkbox State!() -> (
    then
    ControlRig:Control Rig Object Reference
    CheckBox:Check Box Object Reference
    IK Buttons:Array of EUW Picker Button Object References
    FK Buttons:Array of EUW Picker Button Object References
   ) {is=title K2Node_FunctionEntry}
   > n2 = Get Current Local Time(TimeUnit=DisplayRate) -> (ReturnValue_Frame_SubFrame=~, ReturnValue_Time=~, ReturnValue_bHasJumped=~)
   > n3 = GetFocusedLevelSequence()
   > n4 = GetLocalControlRigBool(
    LevelSequence=@n3
    ControlRig=@n1
    ControlName=Convert to MR naming(In name=GetObjectName(Object=@n1.CheckBox)) -> (Out name=None)
    Frame=@n2
    TimeUnit=DisplayRate
   )
   > n5 = Set bIsChecked(bIsChecked=@n4)
   > SetIsChecked!(execute, self:Check Box Object Reference=Get CheckBox() -> (CheckBox:Check Box Object Reference), InIsChecked:bool=@n5) -> (then)
   > Set Visibility for Array(" Array"=Get FK Buttons(), bVisibility=NOT Boolean(A=Get bIsChecked()))
   > Set Visibility for Array(" Array"=Get IK Buttons(), bVisibility=Get bIsChecked())
  ]}
 Set Checkboxes State From Scene {nodes [
   n1 = Set Checkboxes State From Scene!() -> (then, ControlRig:Control Rig Object Reference, Checkboxes:Array of Check Box Object References) {is=title K2Node_FunctionEntry}
   > n2 = GetFocusedLevelSequence()
   > n3 = For Each Loop(Array:Array of Check Box Object References=@n1.Checkboxes) -> (Array Element:Check Box Object Reference)
   > n4 = Get Current Local Time(TimeUnit=DisplayRate) -> (ReturnValue_Frame_SubFrame=~, ReturnValue_Time=~, ReturnValue_bHasJumped=~)
   > n5 = GetLocalControlRigBool(
    LevelSequence=@n2
    ControlRig=Get ControlRig() -> (ControlRig:Control Rig Object Reference)
    ControlName=Convert to MR naming(In name=GetObjectName(Object=@n3)) -> (Out name=None)
    Frame=@n4
    TimeUnit=DisplayRate
   )
   > SetIsChecked(self=@n3, InIsChecked=@n5)
  ]}
 Set IK FK Switch And Widget Visibility {nodes [
   n1 = Set IK FK Switch And Widget Visibility!() -> (
    then
    ControlRig:Control Rig Object Reference
    CheckBox:Check Box Object Reference
    Condition:bool=~
    IK Buttons:Array of EUW Picker Button Object References
    FK Buttons:Array of EUW Picker Button Object References
    Hide Buttons:bool=true
   ) {is=title K2Node_FunctionEntry}
   > Set Animation Channel Bool!(
    execute
    self
    ControlRig:Control Rig Object Reference=@n1
    Channel Name:Name=Convert to MR naming(In name=GetObjectName(Object=@n1.CheckBox)) -> (Out name=None)
    Value:bool=@n1.Condition
   ) -> (then) {is=title K2Node_CallFunction}
   > Branch(Condition=Get Hide Buttons() -> (Hide Buttons:bool))
   > Set Buttons Visibility From Checkbox State!(
    execute
    self
    ControlRig:Control Rig Object Reference=Get ControlRig()
    CheckBox:Check Box Object Reference=Get CheckBox()
    IK Buttons:Array of EUW Picker Button Object References=Get IK Buttons()
    FK Buttons:Array of EUW Picker Button Object References=Get FK Buttons()
   ) -> (then) {is=title K2Node_CallFunction}
  ]}
 Set Visibility for Array {nodes [
   n1 = Set Visibility for Array!() -> (then, " Array":Array of EUW Picker Button Object References, bVisibility:bool=~) {is=title K2Node_FunctionEntry}
   > n2 = For Each Loop(Array:Array of EUW Picker Button Object References=@n1) -> (Array Element:EUW Picker Button Object Reference)
   Return Node(execute=@n2.Completed)
   > n3 = Branch(Condition=Get bVisibility() -> (bVisibility:bool))
   > SetVisibility(self=@n2, InVisibility=Visible)
   SetVisibility(execute=@n3.else, self=@n2, InVisibility=Collapsed)
  ]}
 Setup MR Naming For Buttons {nodes [
   n1 = Setup MR Naming For Buttons!() -> (
    then
    Main Canvas:Panel Widget Object Reference
    Control Rig:Control Rig Object Reference
    Excluded Buttons:Array of Widget Object References
   ) {is=title K2Node_FunctionEntry}
   > n2 = Sequence()
   > Clear(TargetArray=Get Control Names())
   > Clear(TargetArray=Get Shape Names() -> (Shape Names:Array of Names))
   n3 = GetHierarchy(self=Get Control Rig() -> (Control Rig:Control Rig Object Reference))
   > n4 = GetController!(execute, self:Rig Hierarchy Object Reference=@n3, bCreateIfNeeded:bool=true) -> (then, ReturnValue:Rig Hierarchy Controller Object Reference)
   n5 = Get Controls(self=@n3, bTraverse=true)
   > n6 = For Each Loop with Break!(Exec:Exec, Array:Array of Rig Element Key Structures=@n5, Break:Exec) -> (LoopBody:Exec, Array Element_Type:ERigElementType Enum, Array Element_Name:Name, Array Index:int, Completed:Exec)
   > Add(TargetArray:Array of Names=Get Control Names(), NewItem:"Name (by ref)"=@n6.Array Element_Name) {is=title K2Node_CallArrayFunction}
   n7 = "Get (a ref)"(Array:Array of Rig Element Key Structures=@n5, Dimension 1=@n6.Array Index) -> (Output:"Rig Element Key Structure (by ref)")
   > Add(
    TargetArray=Get Shape Names()
    NewItem=Break Rig Control Settings!(RigControlSettings:"Rig Control Settings Structure (by ref)"=GetControlSettings!(self:Rig Hierarchy Controller Object Reference=@n4, InKey:Rig Element Key Structure=@n7) -> (ReturnValue:Rig Control Settings Structure)) -> (DisplayName:Name) {is=title K2Node_BreakStruct}
   )
   n8 = For Each Loop(Exec=@n2.then_1, Array:Array of Widget Object References=GetAllChildren(self=@n1)) -> (Array Element:Widget Object Reference)
   Return Node(execute=@n8.Completed)
   > n9 = Branch(Condition=Contains Item(TargetArray:Array of Widget Object References=@n1.Excluded Buttons, ItemToFind:"Widget Object Reference (by ref)"=@n8))
   n10 = Cast To EUW_Picker_Element(execute=@n9.else, Object=@n8)
   > n11 = For Each Loop with Break!(Exec, Array:Array of Names=Get Control Names(), Break=@n15.then_2) -> (LoopBody, Array Element:Name, Array Index, Completed)
   n12 = GetObjectName(Object=@n10)
   n13 = Replace(SourceString=Replace(SourceString=@n12, From=_, To=/), From=-, To=" ")
   n14 = "To String (Name)"(InName=@n11)
   > Branch(Condition=OR Boolean(A=EndsWith(SourceString=@n14, InSuffix=@n12), B=EndsWith(SourceString=@n14, InSuffix=@n13)))
   > n15 = Sequence!(execute) -> (then_0, then_1, then_2:Exec)
   Branch(
    execute=@n15.then_1
    Condition="Equal Exactly (Text)"!(A:"Text (by ref)"=Get ToolTipText!(self:Widget Object Reference=@n10) -> (ToolTipText:Text), B:"Text (by ref)"="To Text (String)"()) -> (ReturnValue:bool)
   )
   n16 = Get Shape Names()
   > SetToolTipText!(
    execute
    self:Widget Object Reference=@n10
    InToolTipText:"Text (by ref)"="To Text (Name)"!(InName:Name="Get (a ref)"(Array:Array of Names=@n16, Dimension 1=@n11.Array Index) -> (Output:"Name (by ref)")) -> (ReturnValue:Text)
   ) -> (then)
   Set Controls for Primary Control Rig(execute=@n15.then_0, self=@n10, Control Names=Make Array("[0]":Name=@n11) -> (Array:Array of Names))
  ]}
]
variables [
 Anim Picker EUW {type=Animation_Picker_C, default=null}
 Animatable Creature {type=AnimatableCreature_C, default=null}
 Control Rig Proxy Binding {
  type=ControlRigSequencerBindingProxy
  default {controlRig=null, track=null, proxy {bindingId=#2, sequence=null}}
 }
]
widgets {info {rootWidgetClass=null, widgetCount=0}, widgets []}
