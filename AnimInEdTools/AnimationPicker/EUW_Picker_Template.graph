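# Template editor utility widget for an animation picker, derived from EUW_Asset_Picker_Base, with example global and left-leg buttons for a Sequencer Control Rig.
# On construct or load, it takes the bound rig, maps button names to controls and builds IK/FK selection groups.
# Checkboxes write rig channels; the IK/FK switch recolors buttons and resyncs on mouse enter.
# A button toggles the IK/FK matching tool, and a helper function keys preview-mesh tracks.
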EditorUtilityWidgetBlueprint /AnimInEdTools/AnimationPicker/EUW_Picker_Template
graphs [
 Add Preview Mesh Track {nodes [
   Add Preview Mesh Track!() -> (then, Blueprint Anim:Object Class Reference, Preview Mesh:Object Reference, SKM Component:String) {is=title K2Node_FunctionEntry}
   > n1 = Sequence!(execute) -> (then_0, then_1, then_2:Exec)
   > n2 = GetCurrentLevelSequence!(execute) -> (then, ReturnValue:Level Sequence Object Reference)
   > n3 = Set Current Level Sequence(Current Level Sequence:Level Sequence Object Reference=@n2) -> (Output_Get:Level Sequence Object Reference)
   > n4 = GetBindings!(execute, Sequence:Movie Scene Sequence Object Reference=@n3) -> (then, ReturnValue:Array of Movie Scene Binding Proxy Structures)
   > n5 = For Each Loop(Array:Array of Movie Scene Binding Proxy Structures=@n4) -> (Array Element:Movie Scene Binding Proxy Structure)
   > n6 = GetBindingID!(
    execute
    Sequence:Movie Scene Sequence Object Reference=Get Current Level Sequence() -> (Current Level Sequence:Level Sequence Object Reference)
    InBinding:"Movie Scene Binding Proxy Structure (by ref)"=@n5
   ) -> (then, ReturnValue:Movie Scene Object Binding ID Structure)
   > n7 = For Each Loop(Array:Array of Object References=GetBoundObjects!(ObjectBinding:Movie Scene Object Binding ID Structure=@n6) -> (ReturnValue:Array of Object References)) -> (Array Element:Object Reference)
   > n8 = Sequence!(execute) -> (then_0, then_1)
   n9 = GetDisplayName(Object=@n7)
   n10 = Get Blueprint Anim() -> (Blueprint Anim:Object Class Reference)
   > Branch(Condition="Equal Exactly (String)"(A=Replace(SourceString=@n9, From=" ", To=_), B=Get Class Display Name(Class=@n10)))
   > Set BP Spawnable Object(BP Spawnable Object:Object Reference=@n7) -> (Output_Get:Object Reference)
   > Set BP Binding(BP Binding:Movie Scene Binding Proxy Structure=@n5) -> (Output_Get:Movie Scene Binding Proxy Structure)
   n11 = GetDisplayName(Object=@n7)
   n12 = Get Class Display Name(Class=Get Blueprint Anim())
   n13 = Get SKM Component() -> (SKM Component:String)
   Branch(
    execute=@n8.then_1
    Condition=Contains(SearchIn=Replace(SourceString=@n11, From=" ", To=_), Substring=Append!(A=@n12, B=., C:String=@n13) -> (ReturnValue))
   )
   > Remove!(execute, InBinding:"Movie Scene Binding Proxy Structure (by ref)"=@n5) -> (then)
   n14 = Cast To Actor!(execute=@n1.then_1, Object=Get BP Spawnable Object() -> (BP Spawnable Object:Object Reference)) -> (then, CastFailed, AsActor:Actor Object Reference) {is=title K2Node_DynamicCast}
   > n15 = For Each Loop(Array:Array of Skeletal Mesh Component Object References=Get Components By Class!(self:Actor Object Reference=@n14, ComponentClass:Actor Component Class Reference=/Script/Engine.SkeletalMeshComponent) -> (ReturnValue:Array of Skeletal Mesh Component Object References)) -> (Array Element:Skeletal Mesh Component Object Reference)
   > Branch(Condition=Contains(SearchIn=GetDisplayName(Object=@n15), Substring=Get SKM Component()))
   > n16 = AddPossessable!(execute, Sequence:Movie Scene Sequence Object Reference=Get Current Level Sequence(), ObjectToPossess:Object Reference=@n15) -> (then, ReturnValue:Movie Scene Binding Proxy Structure)
   > SetParent!(
    execute
    InBinding:"Movie Scene Binding Proxy Structure (by ref)"=@n16
    InParentBinding:"Movie Scene Binding Proxy Structure (by ref)"=Get BP Binding() -> (BP Binding:Movie Scene Binding Proxy Structure)
   ) -> (then)
   > n17 = AddTrack!(
    execute
    InBinding:"Movie Scene Binding Proxy Structure (by ref)"=@n16
    TrackType:Movie Scene Track Class Reference=/Script/MovieSceneTracks.MovieSceneObjectPropertyTrack
   ) -> (then, ReturnValue:Movie Scene Track Object Reference)
   > n18 = Cast To MovieSceneObjectPropertyTrack!(execute, Object=@n17) -> (then, CastFailed, AsMovie Scene Object Property Track:Movie Scene Object Property Track Object Reference) {is=title K2Node_DynamicCast}
   > SetPropertyNameAndPath!(
    execute
    Track:Movie Scene Property Track Object Reference=@n18
    InPropertyName:"Name (by ref)"=MakeLiteralName(Value=Skeletal Mesh Asset)
    InPropertyPath:String=MakeLiteralString(Value=SkeletalMeshAsset)
   ) -> (then)
   > n19 = AddSection!(execute, Track:Movie Scene Track Object Reference=@n18) -> (then, ReturnValue:Movie Scene Section Object Reference)
   > n20 = GetAllChannels!(execute, Section:Movie Scene Section Object Reference=@n19) -> (then, ReturnValue:Array of Movie Scene Scripting Channel Object References)
   > n21 = For Each Loop(Array:Array of Movie Scene Scripting Channel Object References=@n20) -> (Array Element:Movie Scene Scripting Channel Object Reference)
   > n22 = Cast To MovieSceneScriptingObjectPathChannel!(execute, Object=@n21) -> (then, CastFailed, AsMovie Scene Scripting Object Path Channel:Movie Scene Scripting Object Path Channel Object Reference) {is=title K2Node_DynamicCast}
   > n23 = For Each Loop(Array:Array of Movie Scene Scripting Key Object References="Get Keys (Object Path)"!(self:Movie Scene Scripting Object Path Channel Object Reference=@n22) -> (ReturnValue:Array of Movie Scene Scripting Key Object References) {is=title K2Node_CallFunction}) -> (Array Element:Movie Scene Scripting Key Object Reference)
   > "Remove Key (Object Path)"!(execute, self:Movie Scene Scripting Object Path Channel Object Reference=@n22, Key:Movie Scene Scripting Key Object Reference=@n23) -> (then)
   "Add Key (Object Path)"!(
    execute=@n23.Completed
    self:Movie Scene Scripting Object Path Channel Object Reference=@n22
    InTime_Value:int=0
    NewValue:Object Reference=Get Preview Mesh() -> (Preview Mesh:Object Reference)
    SubFrame:float=0
    TimeUnit:EMovieSceneTimeUnit Enum=DisplayRate
   ) -> (then, ReturnValue:Movie Scene Scripting Object Path Key Object Reference)
  ]}
 Cache IK FK Buttons {nodes [
   Cache IK FK Buttons() {is=title K2Node_FunctionEntry}
   > n1 = Sequence()
   > Set Left Leg FK(Left Leg FK:Array of EUW Picker Button Object References=Make Array!(
     "[0]":EUW Picker Button Object Reference=Get Leg-L_FK-0() -> (Leg-L_FK-0:EUW Picker Button Object Reference)
     "[1]":EUW Picker Button Object Reference=Get Leg-L_FK-1() -> (Leg-L_FK-1:EUW Picker Button Object Reference)
     "[2]":EUW Picker Button Object Reference=Get Leg-L_FK-2() -> (Leg-L_FK-2:EUW Picker Button Object Reference)
    ) -> (Array:Array of EUW Picker Button Object References)) -> (Output_Get:Array of EUW Picker Button Object References)
   Set Left Leg IK(
    execute=@n1.then_1
    Left Leg IK:Array of EUW Picker Button Object References=Make Array!("[0]"=Get Leg-L_PV() -> (Leg-L_PV:EUW Picker Button Object Reference), "[1]"=Get Leg-L_IK() -> (Leg-L_IK:EUW Picker Button Object Reference)) -> (Array)
   ) -> (Output_Get:Array of EUW Picker Button Object References)
   > Append Array!(
    execute
    TargetArray:Array of EUW Picker Button Object References=Get All Leg L() -> (All Leg L:Array of EUW Picker Button Object References)
    SourceArray:Array of EUW Picker Button Object References=Get Left Leg FK() -> (Left Leg FK:Array of EUW Picker Button Object References)
   ) -> (then)
   > Append Array(TargetArray=Get All Leg L(), SourceArray=Get Left Leg IK() -> (Left Leg IK:Array of EUW Picker Button Object References))
  ]}
 EventGraph {nodes [
   n1 = "On Check State Changed (Leg-L_IK-FK-Switch)"!() -> (OutputDelegate, then, bIsChecked:bool) {is=title K2Node_ComponentBoundEvent}
   n2 = Get Leg-L_Ik-Fk-Switch() -> (Leg-L_Ik-Fk-Switch:Editor Utility Check Box Object Reference)
   > Set Buttons Color From Checkbox State!(
    execute
    self:EUW Asset Picker Base Object Reference
    CheckBox:Editor Utility Check Box Object Reference=@n2
    IK Buttons:Array of EUW Picker Button Object References=Get Left Leg IK()
    FK Buttons:Array of EUW Picker Button Object References=Get Left Leg FK()
    Color Mult:double=1
    Opacity:double=0.5
   ) -> (then) {is=title K2Node_CallFunction}
   > Set Animation Channel Bool!(
    execute
    self:EUW Asset Picker Base Object Reference
    ControlRig:Control Rig Object Reference=Get Control Rig() -> (Control Rig:Control Rig Object Reference)
    Channel Name:Name=Convert to MR naming!(In name:String=GetDisplayName(Object=@n2)) -> (Out name:Name=None)
    Value:bool=@n1.bIsChecked
   ) -> (then) {is=title K2Node_CallFunction}
   "On Clicked (Bake_IK_FK_Tool_Button)"() {is=title K2Node_ComponentBoundEvent}
   n3 = Get DialogSlot() -> (DialogSlot:Named Slot Object Reference)
   > n4 = Branch(Condition=HasAnyChildren!(self:Panel Widget Object Reference=@n3) -> (ReturnValue:bool))
   > ClearChildren(self=@n3)
   n5 = Create Widget!(
    execute:Exec=@n4.else
    WorldContextObject:Object Reference
    Class:User Widget Class Reference=/AnimInEdTools/IkFkMatchingTool/EUW_IkFkMatching.EUW_IkFkMatching_C
    OwningPlayer:Player Controller Object Reference
    Primary Control Rig:Control Rig Object Reference=Get Control Rig()
    Animatable Creature:Animatable Creature Object Reference=Get Animatable Creature() -> (Animatable Creature:Animatable Creature Object Reference)
   ) -> (then:Exec, ReturnValue:EUW Ik Fk Matching Object Reference)
   > AddChild(self=Get DialogSlot(), Content=@n5)
   n6 = "On Check State Changed (Capsule_Draw-Capsule)"!() -> (OutputDelegate, then, bIsChecked:bool) {is=title K2Node_ComponentBoundEvent}
   > Set Animation Channel Bool(ControlRig=Get Control Rig(), Channel Name=Capsule/Draw Capsule, Value=@n6.bIsChecked)
   n7 = Event Construct()
   n8 = Get!() -> (
    Control Rig Proxy Binding_Proxy:Movie Scene Binding Proxy Structure
    Control Rig Proxy Binding_ControlRig:Control Rig Object Reference
    Control Rig Proxy Binding_Track:Movie Scene Control Rig Parameter Track Object Reference
   )
   Is Valid(exec=[@n12.then, @n7.then], InputObject=@n8.Control Rig Proxy Binding_ControlRig)
   > n9 = Set Control Rig(Control Rig:Control Rig Object Reference=@n8.Control Rig Proxy Binding_ControlRig) -> (Output_Get:Control Rig Object Reference)
   > Setup MR Naming For Buttons!(
    execute
    self:EUW Asset Picker Base Object Reference
    Main Canvas:Panel Widget Object Reference=Get CanvasPanel() -> (CanvasPanel:Canvas Panel Object Reference)
    Control Rig:Control Rig Object Reference=@n9
    Excluded Buttons:Array of Widget Object References
   ) -> (then)
   > Cache IK FK Buttons() {is=title K2Node_CallFunction}
   > Set Inits Buttons() {is=title K2Node_CallFunction}
   n10 = Event OnMouseEnter()
   > Parent: OnMouseEnter!(execute, MyGeometry:Geometry Structure=@n10.MyGeometry, MouseEvent:"Pointer Event Structure (by ref)"=@n10.MouseEvent) -> (then) {is=title K2Node_CallParentFunction}
   > Set Checkboxes State From Scene!(
    execute
    self:EUW Asset Picker Base Object Reference
    ControlRig:Control Rig Object Reference=Get Control Rig()
    Checkboxes:Array of Check Box Object References=Make Array!("[0]":Check Box Object Reference=Get Leg-L_IK-FK-Switch() -> (Leg-L_IK-FK-Switch:Editor Utility Check Box Object Reference)) -> (Array:Array of Check Box Object References)
   ) -> (then) {is=title K2Node_CallFunction}
   > Set Buttons Color From Checkbox State(CheckBox=Get Leg-L_Ik-Fk-Switch(), IK Buttons=Get Left Leg IK(), FK Buttons=Get Left Leg FK(), Color Mult=1, Opacity=0.5)
   n11 = Event Widget has loaded!() -> (
    OutputDelegate
    then
    Control Rig Proxy Binding:Control Rig Sequencer Binding Proxy Structure
    Animatable Creature:Animatable Creature Object Reference
    Anim Picker EUW:Animation Picker Object Reference
   ) {is=title K2Node_Event}
   > n12 = Parent: Widget has loaded!(
    execute
    Control Rig Proxy Binding:Control Rig Sequencer Binding Proxy Structure=@n11.Control Rig Proxy Binding
    Animatable Creature:Animatable Creature Object Reference=@n11.Animatable Creature
    Anim Picker EUW:Animation Picker Object Reference=@n11.Anim Picker EUW
   ) -> (then) {is=title K2Node_CallParentFunction}
   Event Update From Scene() {is=title K2Node_Event}
   > Parent: Update From Scene() {is=title K2Node_CallParentFunction}
  ]}
 Get Primary Panel {nodes [
   Get Primary Panel() {is=title K2Node_FunctionEntry}
   > Return Node!(execute, Panel:Panel Widget Object Reference=Get CanvasPanel())
  ]}
 Set Inits Buttons {nodes [
   Set Inits Buttons() {is=title K2Node_FunctionEntry}
   > Select Multiple Buttons!(
    execute
    self:EUW Asset Picker Base Object Reference
    Button:EUW Picker Button Object Reference=Get All_L_Legs_FKs() -> (All_L_Legs_FKs:EUW Picker Button Object Reference)
    Buttons to Select:Array of EUW Picker Button Object References=Get Left Leg FK()
   ) -> (then)
   > Select Multiple Buttons(Button=Get All_L_Legs_IKs() -> (All_L_Legs_IKs:EUW Picker Button Object Reference), Buttons to Select=Get Left Leg IK())
   > Select Multiple Buttons(Button=Get All_L_Leg() -> (All_L_Leg:EUW Picker Button Object Reference), Buttons to Select=Get All Leg L())
  ]}
]
variables [
 All Leg L {type="array<EUW_Picker_Button_C>", default []}
 Control Rig {type=ControlRig, default=null}
 Left Leg FK {type="array<EUW_Picker_Button_C>", default []}
 Left Leg IK {type="array<EUW_Picker_Button_C>", default []}
 Level Sequence {type=LevelSequence, default=null}
]
widgets {
 info {
  parentClass=&/EUW_Asset_Picker_Base.EUW_Asset_Picker_Base_C
  rootWidgetClass=&CanvasPanel
  widgetCount=19
 }
 widgets [
  CanvasPanel {widgetClassPath=&CanvasPanel, bIsVariable=true, widget=CanvasPanel}
  Global {
   widgetClassPath=&/EUW_Picker_Button
   bIsVariable=true
   parent=CanvasPanel
   widget=Global
   properties {
    draw Style=Square
    label Font Size=30
    label Text=G
    label Color {colorUseRule=UseColor_Specified, specifiedColor=<rgba 0 0 0 1>}
    normal Tint {colorUseRule=UseColor_Specified, specifiedColor=<rgba 0.9734 0.4793 0.0423 1>}
   }
   slot {
    class=CanvasPanelSlot
    properties {zOrder=-1, layoutData {alignment=<0 0>, anchors {maximum=<0 0>, minimum=<0 0>}, offsets=<ltrb 28 28 60 60>}}
   }
  }
  Local {
   widgetClassPath=&EUW_Picker_Button
   bIsVariable=true
   parent=CanvasPanel
   widget=Local
   properties {
    draw Style=Square
    label Font Size=30
    label Text=L
    label Color {colorUseRule=UseColor_Specified, specifiedColor=<rgba 0 0 0 1>}
    normal Tint {colorUseRule=UseColor_Specified, specifiedColor=<rgba 0.9734 0.4793 0.0423 1>}
   }
   slot {class=CanvasPanelSlot, properties {layoutData {alignment=<0 0>, anchors {maximum=<0 0>, minimum=<0 0>}, offsets=<ltrb 112 28 60 60>}}}
  }
  Root {
   widgetClassPath=&EUW_Picker_Button
   bIsVariable=true
   parent=CanvasPanel
   widget=Root
   properties {
    draw Style=Square
    label Font Size=30
    label Text=R
    label Color {colorUseRule=UseColor_Specified, specifiedColor=<rgba 0 0 0 1>}
    normal Tint {colorUseRule=UseColor_Specified, specifiedColor=<rgba 0.9734 0.4793 0.0423 1>}
   }
   slot {
    class=CanvasPanelSlot
    properties {zOrder=1, layoutData {alignment=<0 0>, anchors {maximum=<0 0>, minimum=<0 0>}, offsets=<ltrb 196 28 60 60>}}
   }
  }
  Leg-L_FK-0 {
   widgetClassPath=&EUW_Picker_Button
   bIsVariable=true
   parent=CanvasPanel
   widget=Leg-L_FK-0
   properties {draw Style=Square, normal Tint {colorUseRule=UseColor_Specified, specifiedColor=<rgba 0.0273 0.0666 0.7913 1>}}
   slot {
    class=CanvasPanelSlot
    properties {zOrder=1, layoutData {alignment=<0 0>, anchors {maximum=<0 0>, minimum=<0 0>}, offsets=<ltrb 480 324 39.6533 141.8906>}}
   }
  }
  Leg-L_FK-1 {
   widgetClassPath=&EUW_Picker_Button
   bIsVariable=true
   parent=CanvasPanel
   widget=Leg-L_FK-1
   properties {draw Style=Square, normal Tint {colorUseRule=UseColor_Specified, specifiedColor=<rgba 0.0273 0.0666 0.7913 1>}}
   slot {
    class=CanvasPanelSlot
    properties {zOrder=1, layoutData {alignment=<0 0>, anchors {maximum=<0 0>, minimum=<0 0>}, offsets=<ltrb 480 480.6633 39.4964 152.1898>}}
   }
  }
  Leg-L_FK-2 {
   widgetClassPath=&EUW_Picker_Button
   bIsVariable=true
   parent=CanvasPanel
   widget=Leg-L_FK-2
   properties {
    draw Style=Square
    label Font Size=15
    label Text=FK
    label Color {colorUseRule=UseColor_Specified, specifiedColor=<rgba 0 0 0 1>}
    normal Tint {colorUseRule=UseColor_Specified, specifiedColor=<rgba 0.0273 0.0666 0.7913 1>}
   }
   slot {
    class=CanvasPanelSlot
    properties {zOrder=1, layoutData {alignment=<0 0>, anchors {maximum=<0 0>, minimum=<0 0>}, offsets=<ltrb 480 644 100.7243 40>}}
   }
  }
  All_L_Legs_FKs {
   widgetClassPath=&EUW_Picker_Button
   bIsVariable=true
   parent=CanvasPanel
   widget=All_L_Legs_FKs
   properties {
    draw Style=Square
    label Font Size=15
    label Text=FKs
    label Color {colorUseRule=UseColor_Specified, specifiedColor=<rgba 0 0 0 1>}
    normal Tint {colorUseRule=UseColor_Specified, specifiedColor=<rgba 0.1274 0.5711 0.1144 1>}
   }
   slot {
    class=CanvasPanelSlot
    properties {zOrder=1, layoutData {alignment=<0 0>, anchors {maximum=<0 0>, minimum=<0 0>}, offsets=<ltrb 592 644 100.7243 40>}}
   }
  }
  Leg-L_IK {
   widgetClassPath=&EUW_Picker_Button
   bIsVariable=true
   parent=CanvasPanel
   widget=Leg-L_IK
   properties {
    draw Style=Square
    label Font Size=15
    label Text=IK
    label Color {colorUseRule=UseColor_Specified, specifiedColor=<rgba 0 0 0 1>}
    normal Tint {colorUseRule=UseColor_Specified, specifiedColor=<rgba 0.2232 0.2542 0.7758 1>}
   }
   slot {
    class=CanvasPanelSlot
    properties {zOrder=1, layoutData {alignment=<0 0>, anchors {maximum=<0 0>, minimum=<0 0>}, offsets=<ltrb 480 696 100.1263 40>}}
   }
  }
  All_L_Leg {
   widgetClassPath=&EUW_Picker_Button
   bIsVariable=true
   parent=CanvasPanel
   widget=All_L_Leg
   properties {
    draw Style=Square
    label Font Size=15
    label Text=Leg L
    label Color {colorUseRule=UseColor_Specified, specifiedColor=<rgba 0 0 0 1>}
    normal Tint {colorUseRule=UseColor_Specified, specifiedColor=<rgba 0.117 0.2623 0.1095 1>}
   }
   slot {
    class=CanvasPanelSlot
    properties {zOrder=1, layoutData {alignment=<0 0>, anchors {maximum=<0 0>, minimum=<0 0>}, offsets=<ltrb 480 748 212.5244 40>}}
   }
  }
  All_L_Legs_IKs {
   widgetClassPath=&EUW_Picker_Button
   bIsVariable=true
   parent=CanvasPanel
   widget=All_L_Legs_IKs
   properties {
    draw Style=Square
    label Font Size=15
    label Text=IKs
    label Color {colorUseRule=UseColor_Specified, specifiedColor=<rgba 0 0 0 1>}
    normal Tint {colorUseRule=UseColor_Specified, specifiedColor=<rgba 0.2582 0.5906 0.2582 1>}
   }
   slot {
    class=CanvasPanelSlot
    properties {zOrder=1, layoutData {alignment=<0 0>, anchors {maximum=<0 0>, minimum=<0 0>}, offsets=<ltrb 592 696 100.1263 40>}}
   }
  }
  Leg-L_PV {
   widgetClassPath=&EUW_Picker_Button
   bIsVariable=true
   parent=CanvasPanel
   widget=Leg-L_PV
   properties {
    draw Style=Square
    label Font Size=15
    label Text=PV
    label Color {colorUseRule=UseColor_Specified, specifiedColor=<rgba 0 0 0 1>}
    normal Tint {colorUseRule=UseColor_Specified, specifiedColor=<rgba 0.2232 0.2542 0.7758 1>}
   }
   slot {
    class=CanvasPanelSlot
    properties {zOrder=1, layoutData {alignment=<0 0>, anchors {maximum=<0 0>, minimum=<0 0>}, offsets=<ltrb 540 452 52 52>}}
   }
  }
  Leg-L_IK-FK-Switch {
   widgetClassPath=&EditorUtilityCheckBox
   bIsVariable=true
   parent=CanvasPanel
   widget=Leg-L_IK-FK-Switch
   properties {
    bOverride_Cursor=true
    checkedState=Checked
    renderTransform {angle=0, scale=<2 2>, shear=<0 0>, translation=<0 0>}
    widgetStyle {
     checkBoxType=CheckBox
     backgroundHoveredImage {
      bIsDynamicallyLoaded=false
      drawAs=RoundedBox
      imageType=NoImage
      mirroring=NoMirror
      resourceName=None
      resourceObject=null
      tiling=NoTile
      imageSize=<18 18>
      margin=<ltrb 0 0 0 0>
      outlineSettings {
       bUseBrushTransparency=false
       roundingType=FixedRadius
       width=1
       color {colorUseRule=UseColor_ColorTable, specifiedColor=<rgba 0 0 0 0>}
       cornerRadii=<3 3 3 3>
      }
      tintColor {colorUseRule=UseColor_Specified, specifiedColor=<rgba 0.0048 0.0048 0.0048 1>}
      uVRegion {bIsValid=false, max=<0 0>, min=<0 0>}
     }
     backgroundImage {
      bIsDynamicallyLoaded=false
      drawAs=RoundedBox
      imageType=NoImage
      mirroring=NoMirror
      resourceName=None
      resourceObject=null
      tiling=NoTile
      imageSize=<18 18>
      margin=<ltrb 0 0 0 0>
      outlineSettings {
       bUseBrushTransparency=false
       roundingType=FixedRadius
       width=1
       color {colorUseRule=UseColor_ColorTable, specifiedColor=<rgba 0 0 0 0>}
       cornerRadii=<3 3 3 3>
      }
      tintColor {colorUseRule=UseColor_Specified, specifiedColor=<rgba 0.0048 0.0048 0.0048 1>}
      uVRegion {bIsValid=false, max=<0 0>, min=<0 0>}
     }
     backgroundPressedImage {
      bIsDynamicallyLoaded=false
      drawAs=RoundedBox
      imageType=NoImage
      mirroring=NoMirror
      resourceName=None
      resourceObject=null
      tiling=NoTile
      imageSize=<18 18>
      margin=<ltrb 0 0 0 0>
      outlineSettings {
       bUseBrushTransparency=false
       roundingType=FixedRadius
       width=1
       color {colorUseRule=UseColor_ColorTable, specifiedColor=<rgba 0 0 0 0>}
       cornerRadii=<3 3 3 3>
      }
      tintColor {colorUseRule=UseColor_Specified, specifiedColor=<rgba 0.0048 0.0048 0.0048 1>}
      uVRegion {bIsValid=false, max=<0 0>, min=<0 0>}
     }
     borderBackgroundColor {colorUseRule=UseColor_Specified, specifiedColor=<rgba 1 1 1 1>}
     checkedForeground {colorUseRule=UseColor_Specified, specifiedColor=<rgba 1 1 1 1>}
     checkedHoveredForeground {colorUseRule=UseColor_Specified, specifiedColor=<rgba 1 1 1 1>}
     checkedHoveredImage {
      bIsDynamicallyLoaded=false
      drawAs=Image
      imageType=Vector
      mirroring=NoMirror
      resourceName=../../../Engine/Content/Slate/Starship/CoreWidgets/CheckBox/check.svg
      resourceObject=null
      tiling=NoTile
      imageSize=<16 16>
      margin=<ltrb 0 0 0 0>
      outlineSettings {
       bUseBrushTransparency=false
       roundingType=HalfHeightRadius
       width=0
       color {colorUseRule=UseColor_Specified, specifiedColor=<rgba 0 0 0 0>}
       cornerRadii=<0 0 0 0>
      }
      tintColor {colorUseRule=UseColor_Specified, specifiedColor=<rgba 0.0044 0.2384 1 1>}
      uVRegion {bIsValid=false, max=<0 0>, min=<0 0>}
     }
     checkedImage {
      bIsDynamicallyLoaded=false
      drawAs=Image
      imageType=Vector
      mirroring=NoMirror
      resourceName=../../../Engine/Content/Slate/Starship/CoreWidgets/CheckBox/check.svg
      resourceObject=null
      tiling=NoTile
      imageSize=<16 16>
      margin=<ltrb 0 0 0 0>
      outlineSettings {
       bUseBrushTransparency=false
       roundingType=HalfHeightRadius
       width=0
       color {colorUseRule=UseColor_Specified, specifiedColor=<rgba 0 0 0 0>}
       cornerRadii=<0 0 0 0>
      }
      tintColor {colorUseRule=UseColor_Specified, specifiedColor=<rgba 0 0.162 0.7454 1>}
      uVRegion {bIsValid=false, max=<0 0>, min=<0 0>}
     }
     checkedPressedForeground {colorUseRule=UseColor_Specified, specifiedColor=<rgba 1 1 1 1>}
     checkedPressedImage {
      bIsDynamicallyLoaded=false
      drawAs=Image
      imageType=Vector
      mirroring=NoMirror
      resourceName=../../../Engine/Content/Slate/Starship/CoreWidgets/CheckBox/check.svg
      resourceObject=null
      tiling=NoTile
      imageSize=<16 16>
      margin=<ltrb 0 0 0 0>
      outlineSettings {
       bUseBrushTransparency=false
       roundingType=HalfHeightRadius
       width=0
       color {colorUseRule=UseColor_Specified, specifiedColor=<rgba 0 0 0 0>}
       cornerRadii=<0 0 0 0>
      }
      tintColor {colorUseRule=UseColor_Specified, specifiedColor=<rgba 0 0.0802 0.3515 1>}
      uVRegion {bIsValid=false, max=<0 0>, min=<0 0>}
     }
     checkedSlateSound {resourceObject=null}
     foregroundColor {colorUseRule=UseColor_Specified, specifiedColor=<rgba 1 1 1 1>}
     hoveredForeground {colorUseRule=UseColor_Specified, specifiedColor=<rgba 1 1 1 1>}
     hoveredSlateSound {resourceObject=null}
     padding=<ltrb 2 0 0 0>
     pressedForeground {colorUseRule=UseColor_Specified, specifiedColor=<rgba 1 1 1 1>}
     uncheckedHoveredImage {
      bIsDynamicallyLoaded=false
      drawAs=NoDrawType
      imageType=NoImage
      mirroring=NoMirror
      resourceName=None
      resourceObject=null
      tiling=NoTile
      imageSize=<0 0>
      margin=<ltrb 0 0 0 0>
      outlineSettings {
       bUseBrushTransparency=false
       roundingType=HalfHeightRadius
       width=0
       color {colorUseRule=UseColor_Specified, specifiedColor=<rgba 0 0 0 0>}
       cornerRadii=<0 0 0 0>
      }
      tintColor {colorUseRule=UseColor_Specified, specifiedColor=<rgba 1 1 1 1>}
      uVRegion {bIsValid=false, max=<0 0>, min=<0 0>}
     }
     uncheckedImage {
      bIsDynamicallyLoaded=false
      drawAs=NoDrawType
      imageType=NoImage
      mirroring=NoMirror
      resourceName=None
      resourceObject=null
      tiling=NoTile
      imageSize=<0 0>
      margin=<ltrb 0 0 0 0>
      outlineSettings {
       bUseBrushTransparency=false
       roundingType=HalfHeightRadius
       width=0
       color {colorUseRule=UseColor_Specified, specifiedColor=<rgba 0 0 0 0>}
       cornerRadii=<0 0 0 0>
      }
      tintColor {colorUseRule=UseColor_Specified, specifiedColor=<rgba 1 1 1 1>}
      uVRegion {bIsValid=false, max=<0 0>, min=<0 0>}
     }
     uncheckedPressedImage {
      bIsDynamicallyLoaded=false
      drawAs=NoDrawType
      imageType=NoImage
      mirroring=NoMirror
      resourceName=None
      resourceObject=null
      tiling=NoTile
      imageSize=<0 0>
      margin=<ltrb 0 0 0 0>
      outlineSettings {
       bUseBrushTransparency=false
       roundingType=HalfHeightRadius
       width=0
       color {colorUseRule=UseColor_Specified, specifiedColor=<rgba 0 0 0 0>}
       cornerRadii=<0 0 0 0>
      }
      tintColor {colorUseRule=UseColor_Specified, specifiedColor=<rgba 1 1 1 1>}
      uVRegion {bIsValid=false, max=<0 0>, min=<0 0>}
     }
     uncheckedSlateSound {resourceObject=null}
     undeterminedForeground {colorUseRule=UseColor_Specified, specifiedColor=<rgba 1 1 1 1>}
     undeterminedHoveredImage {
      bIsDynamicallyLoaded=false
      drawAs=Image
      imageType=Vector
      mirroring=NoMirror
      resourceName=../../../Engine/Content/Slate/Starship/CoreWidgets/CheckBox/indeterminate.svg
      resourceObject=null
      tiling=NoTile
      imageSize=<16 16>
      margin=<ltrb 0 0 0 0>
      outlineSettings {
       bUseBrushTransparency=false
       roundingType=HalfHeightRadius
       width=0
       color {colorUseRule=UseColor_Specified, specifiedColor=<rgba 0 0 0 0>}
       cornerRadii=<0 0 0 0>
      }
      tintColor {colorUseRule=UseColor_Specified, specifiedColor=<rgba 0.0044 0.2384 1 1>}
      uVRegion {bIsValid=false, max=<0 0>, min=<0 0>}
     }
     undeterminedImage {
      bIsDynamicallyLoaded=false
      drawAs=Image
      imageType=Vector
      mirroring=NoMirror
      resourceName=../../../Engine/Content/Slate/Starship/CoreWidgets/CheckBox/indeterminate.svg
      resourceObject=null
      tiling=NoTile
      imageSize=<16 16>
      margin=<ltrb 0 0 0 0>
      outlineSettings {
       bUseBrushTransparency=false
       roundingType=HalfHeightRadius
       width=0
       color {colorUseRule=UseColor_Specified, specifiedColor=<rgba 0 0 0 0>}
       cornerRadii=<0 0 0 0>
      }
      tintColor {colorUseRule=UseColor_Specified, specifiedColor=<rgba 0 0.162 0.7454 1>}
      uVRegion {bIsValid=false, max=<0 0>, min=<0 0>}
     }
     undeterminedPressedImage {
      bIsDynamicallyLoaded=false
      drawAs=Image
      imageType=Vector
      mirroring=NoMirror
      resourceName=../../../Engine/Content/Slate/Starship/CoreWidgets/CheckBox/indeterminate.svg
      resourceObject=null
      tiling=NoTile
      imageSize=<16 16>
      margin=<ltrb 0 0 0 0>
      outlineSettings {
       bUseBrushTransparency=false
       roundingType=HalfHeightRadius
       width=0
       color {colorUseRule=UseColor_Specified, specifiedColor=<rgba 0 0 0 0>}
       cornerRadii=<0 0 0 0>
      }
      tintColor {colorUseRule=UseColor_Specified, specifiedColor=<rgba 0 0.0802 0.3515 1>}
      uVRegion {bIsValid=false, max=<0 0>, min=<0 0>}
     }
    }
   }
   slot {
    class=CanvasPanelSlot
    properties {layoutData {alignment=<0 0>, anchors {maximum=<0 0>, minimum=<0 0>}, offsets=<ltrb 636 560 34.2923 18.6787>}}
   }
  }
  TextBlock_5 {
   widgetClassPath=&TextBlock
   parent=Leg-L_IK-FK-Switch
   widget=TextBlock_5
   properties {
    autoWrapText=true
    justification=Center
    text=IK
    font {
     bForceMonospaced=false
     bMaterialIsStencil=false
     fontMaterial=null
     fontObject=&Roboto
     letterSpacing=0
     monospacedWidth=1
     size=10
     skewAmount=0
     typefaceFontName=Bold
     outlineSettings {
      bApplyOutlineToDropShadows=false
      bMiteredCorners=false
      bSeparateFillAlpha=false
      outlineMaterial=null
      outlineSize=0
      outlineColor=<rgba 0 0 0 1>
     }
    }
   }
   slot {class=PanelSlot}
  }
  Capsule_Draw-Capsule {
   widgetClassPath=/Script/UMG.CheckBox
   bIsVariable=true
   parent=CanvasPanel
   widget=Capsule_Draw-Capsule
   properties {
    visibility=Hidden
    renderTransform {angle=0, scale=<2 2>, shear=<0 0>, translation=<0 0>}
    widgetStyle {
     checkBoxType=CheckBox
     backgroundHoveredImage {
      bIsDynamicallyLoaded=false
      drawAs=RoundedBox
      imageType=NoImage
      mirroring=NoMirror
      resourceName=None
      resourceObject=null
      tiling=NoTile
      imageSize=<18 18>
      margin=<ltrb 0 0 0 0>
      outlineSettings {
       bUseBrushTransparency=false
       roundingType=FixedRadius
       width=1
       color {colorUseRule=UseColor_ColorTable, specifiedColor=<rgba 0 0 0 0>}
       cornerRadii=<3 3 3 3>
      }
      tintColor {colorUseRule=UseColor_Specified, specifiedColor=<rgba 0.0048 0.0048 0.0048 1>}
      uVRegion {bIsValid=false, max=<0 0>, min=<0 0>}
     }
     backgroundImage {
      bIsDynamicallyLoaded=false
      drawAs=RoundedBox
      imageType=NoImage
      mirroring=NoMirror
      resourceName=None
      resourceObject=null
      tiling=NoTile
      imageSize=<18 18>
      margin=<ltrb 0 0 0 0>
      outlineSettings {
       bUseBrushTransparency=false
       roundingType=FixedRadius
       width=1
       color {colorUseRule=UseColor_ColorTable, specifiedColor=<rgba 0 0 0 0>}
       cornerRadii=<3 3 3 3>
      }
      tintColor {colorUseRule=UseColor_Specified, specifiedColor=<rgba 0.0048 0.0048 0.0048 1>}
      uVRegion {bIsValid=false, max=<0 0>, min=<0 0>}
     }
     backgroundPressedImage {
      bIsDynamicallyLoaded=false
      drawAs=RoundedBox
      imageType=NoImage
      mirroring=NoMirror
      resourceName=None
      resourceObject=null
      tiling=NoTile
      imageSize=<18 18>
      margin=<ltrb 0 0 0 0>
      outlineSettings {
       bUseBrushTransparency=false
       roundingType=FixedRadius
       width=1
       color {colorUseRule=UseColor_ColorTable, specifiedColor=<rgba 0 0 0 0>}
       cornerRadii=<3 3 3 3>
      }
      tintColor {colorUseRule=UseColor_Specified, specifiedColor=<rgba 0.0048 0.0048 0.0048 1>}
      uVRegion {bIsValid=false, max=<0 0>, min=<0 0>}
     }
     borderBackgroundColor {colorUseRule=UseColor_Specified, specifiedColor=<rgba 1 1 1 1>}
     checkedForeground {colorUseRule=UseColor_Specified, specifiedColor=<rgba 1 1 1 1>}
     checkedHoveredForeground {colorUseRule=UseColor_Specified, specifiedColor=<rgba 1 1 1 1>}
     checkedHoveredImage {
      bIsDynamicallyLoaded=false
      drawAs=Image
      imageType=Vector
      mirroring=NoMirror
      resourceName=../../../Engine/Content/Slate/Starship/CoreWidgets/CheckBox/check.svg
      resourceObject=null
      tiling=NoTile
      imageSize=<16 16>
      margin=<ltrb 0 0 0 0>
      outlineSettings {
       bUseBrushTransparency=false
       roundingType=HalfHeightRadius
       width=0
       color {colorUseRule=UseColor_Specified, specifiedColor=<rgba 0 0 0 0>}
       cornerRadii=<0 0 0 0>
      }
      tintColor {colorUseRule=UseColor_Specified, specifiedColor=<rgba 0.0044 0.2384 1 1>}
      uVRegion {bIsValid=false, max=<0 0>, min=<0 0>}
     }
     checkedImage {
      bIsDynamicallyLoaded=false
      drawAs=Image
      imageType=Vector
      mirroring=NoMirror
      resourceName=../../../Engine/Content/Slate/Starship/CoreWidgets/CheckBox/check.svg
      resourceObject=null
      tiling=NoTile
      imageSize=<16 16>
      margin=<ltrb 0 0 0 0>
      outlineSettings {
       bUseBrushTransparency=false
       roundingType=HalfHeightRadius
       width=0
       color {colorUseRule=UseColor_Specified, specifiedColor=<rgba 0 0 0 0>}
       cornerRadii=<0 0 0 0>
      }
      tintColor {colorUseRule=UseColor_Specified, specifiedColor=<rgba 0 0.162 0.7454 1>}
      uVRegion {bIsValid=false, max=<0 0>, min=<0 0>}
     }
     checkedPressedForeground {colorUseRule=UseColor_Specified, specifiedColor=<rgba 1 1 1 1>}
     checkedPressedImage {
      bIsDynamicallyLoaded=false
      drawAs=Image
      imageType=Vector
      mirroring=NoMirror
      resourceName=../../../Engine/Content/Slate/Starship/CoreWidgets/CheckBox/check.svg
      resourceObject=null
      tiling=NoTile
      imageSize=<16 16>
      margin=<ltrb 0 0 0 0>
      outlineSettings {
       bUseBrushTransparency=false
       roundingType=HalfHeightRadius
       width=0
       color {colorUseRule=UseColor_Specified, specifiedColor=<rgba 0 0 0 0>}
       cornerRadii=<0 0 0 0>
      }
      tintColor {colorUseRule=UseColor_Specified, specifiedColor=<rgba 0 0.0802 0.3515 1>}
      uVRegion {bIsValid=false, max=<0 0>, min=<0 0>}
     }
     checkedSlateSound {resourceObject=null}
     foregroundColor {colorUseRule=UseColor_Specified, specifiedColor=<rgba 1 1 1 1>}
     hoveredForeground {colorUseRule=UseColor_Specified, specifiedColor=<rgba 1 1 1 1>}
     hoveredSlateSound {resourceObject=null}
     padding=<ltrb 2 0 0 0>
     pressedForeground {colorUseRule=UseColor_Specified, specifiedColor=<rgba 1 1 1 1>}
     uncheckedHoveredImage {
      bIsDynamicallyLoaded=false
      drawAs=NoDrawType
      imageType=NoImage
      mirroring=NoMirror
      resourceName=None
      resourceObject=null
      tiling=NoTile
      imageSize=<0 0>
      margin=<ltrb 0 0 0 0>
      outlineSettings {
       bUseBrushTransparency=false
       roundingType=HalfHeightRadius
       width=0
       color {colorUseRule=UseColor_Specified, specifiedColor=<rgba 0 0 0 0>}
       cornerRadii=<0 0 0 0>
      }
      tintColor {colorUseRule=UseColor_Specified, specifiedColor=<rgba 1 1 1 1>}
      uVRegion {bIsValid=false, max=<0 0>, min=<0 0>}
     }
     uncheckedImage {
      bIsDynamicallyLoaded=false
      drawAs=NoDrawType
      imageType=NoImage
      mirroring=NoMirror
      resourceName=None
      resourceObject=null
      tiling=NoTile
      imageSize=<0 0>
      margin=<ltrb 0 0 0 0>
      outlineSettings {
       bUseBrushTransparency=false
       roundingType=HalfHeightRadius
       width=0
       color {colorUseRule=UseColor_Specified, specifiedColor=<rgba 0 0 0 0>}
       cornerRadii=<0 0 0 0>
      }
      tintColor {colorUseRule=UseColor_Specified, specifiedColor=<rgba 1 1 1 1>}
      uVRegion {bIsValid=false, max=<0 0>, min=<0 0>}
     }
     uncheckedPressedImage {
      bIsDynamicallyLoaded=false
      drawAs=NoDrawType
      imageType=NoImage
      mirroring=NoMirror
      resourceName=None
      resourceObject=null
      tiling=NoTile
      imageSize=<0 0>
      margin=<ltrb 0 0 0 0>
      outlineSettings {
       bUseBrushTransparency=false
       roundingType=HalfHeightRadius
       width=0
       color {colorUseRule=UseColor_Specified, specifiedColor=<rgba 0 0 0 0>}
       cornerRadii=<0 0 0 0>
      }
      tintColor {colorUseRule=UseColor_Specified, specifiedColor=<rgba 1 1 1 1>}
      uVRegion {bIsValid=false, max=<0 0>, min=<0 0>}
     }
     uncheckedSlateSound {resourceObject=null}
     undeterminedForeground {colorUseRule=UseColor_Specified, specifiedColor=<rgba 1 1 1 1>}
     undeterminedHoveredImage {
      bIsDynamicallyLoaded=false
      drawAs=Image
      imageType=Vector
      mirroring=NoMirror
      resourceName=../../../Engine/Content/Slate/Starship/CoreWidgets/CheckBox/indeterminate.svg
      resourceObject=null
      tiling=NoTile
      imageSize=<16 16>
      margin=<ltrb 0 0 0 0>
      outlineSettings {
       bUseBrushTransparency=false
       roundingType=HalfHeightRadius
       width=0
       color {colorUseRule=UseColor_Specified, specifiedColor=<rgba 0 0 0 0>}
       cornerRadii=<0 0 0 0>
      }
      tintColor {colorUseRule=UseColor_Specified, specifiedColor=<rgba 0.0044 0.2384 1 1>}
      uVRegion {bIsValid=false, max=<0 0>, min=<0 0>}
     }
     undeterminedImage {
      bIsDynamicallyLoaded=false
      drawAs=Image
      imageType=Vector
      mirroring=NoMirror
      resourceName=../../../Engine/Content/Slate/Starship/CoreWidgets/CheckBox/indeterminate.svg
      resourceObject=null
      tiling=NoTile
      imageSize=<16 16>
      margin=<ltrb 0 0 0 0>
      outlineSettings {
       bUseBrushTransparency=false
       roundingType=HalfHeightRadius
       width=0
       color {colorUseRule=UseColor_Specified, specifiedColor=<rgba 0 0 0 0>}
       cornerRadii=<0 0 0 0>
      }
      tintColor {colorUseRule=UseColor_Specified, specifiedColor=<rgba 0 0.162 0.7454 1>}
      uVRegion {bIsValid=false, max=<0 0>, min=<0 0>}
     }
     undeterminedPressedImage {
      bIsDynamicallyLoaded=false
      drawAs=Image
      imageType=Vector
      mirroring=NoMirror
      resourceName=../../../Engine/Content/Slate/Starship/CoreWidgets/CheckBox/indeterminate.svg
      resourceObject=null
      tiling=NoTile
      imageSize=<16 16>
      margin=<ltrb 0 0 0 0>
      outlineSettings {
       bUseBrushTransparency=false
       roundingType=HalfHeightRadius
       width=0
       color {colorUseRule=UseColor_Specified, specifiedColor=<rgba 0 0 0 0>}
       cornerRadii=<0 0 0 0>
      }
      tintColor {colorUseRule=UseColor_Specified, specifiedColor=<rgba 0 0.0802 0.3515 1>}
      uVRegion {bIsValid=false, max=<0 0>, min=<0 0>}
     }
    }
   }
   slot {
    class=CanvasPanelSlot
    properties {layoutData {alignment=<0 0>, anchors {maximum=<0 0>, minimum=<0 0>}, offsets=<ltrb 92 124 117.6577 20.6807>}}
   }
  }
  TextBlock_10 {
   widgetClassPath=&TextBlock
   parent=Capsule_Draw-Capsule
   widget=TextBlock_10
   properties {
    justification=Center
    text=Draw Gameplay Capsule
    font {
     bForceMonospaced=false
     bMaterialIsStencil=false
     fontMaterial=null
     fontObject=&Roboto
     letterSpacing=0
     monospacedWidth=1
     size=6.38
     skewAmount=0
     typefaceFontName=Bold
     outlineSettings {
      bApplyOutlineToDropShadows=false
      bMiteredCorners=false
      bSeparateFillAlpha=false
      outlineMaterial=null
      outlineSize=0
      outlineColor=<rgba 0 0 0 1>
     }
    }
   }
   slot {class=PanelSlot}
  }
  Bake_IK_FK_Tool_Button {
   widgetClassPath=&Button
   parent=CanvasPanel
   widget=Bake_IK_FK_Tool_Button
   properties {widgetStyle {
     clickedSlateSound {resourceObject=null}
     disabled {
      bIsDynamicallyLoaded=false
      drawAs=RoundedBox
      imageType=NoImage
      mirroring=NoMirror
      resourceName=None
      resourceObject=null
      tiling=NoTile
      imageSize=<0 0>
      margin=<ltrb 0 0 0 0>
      outlineSettings {
       bUseBrushTransparency=false
       roundingType=FixedRadius
       width=1
       color {colorUseRule=UseColor_ColorTable, specifiedColor=<rgba 0 0 0 0>}
       cornerRadii=<4 4 4 4>
      }
      tintColor {colorUseRule=UseColor_Specified, specifiedColor=<rgba 0.0395 0.0395 0.0395 1>}
      uVRegion {bIsValid=false, max=<0 0>, min=<0 0>}
     }
     disabledForeground {colorUseRule=UseColor_ColorTable, specifiedColor=<rgba 0 0 0 0>}
     hovered {
      bIsDynamicallyLoaded=false
      drawAs=RoundedBox
      imageType=NoImage
      mirroring=NoMirror
      resourceName=None
      resourceObject=null
      tiling=NoTile
      imageSize=<32 32>
      margin=<ltrb 0 0 0 0>
      outlineSettings {
       bUseBrushTransparency=false
       roundingType=FixedRadius
       width=1
       color {colorUseRule=UseColor_ColorTable, specifiedColor=<rgba 0 0 0 0>}
       cornerRadii=<4 4 4 4>
      }
      tintColor {colorUseRule=UseColor_Specified, specifiedColor=<rgba 0.8419 0.2914 0.0757 1>}
      uVRegion {bIsValid=false, max=<0 0>, min=<0 0>}
     }
     hoveredForeground {colorUseRule=UseColor_ColorTable, specifiedColor=<rgba 0 0 0 0>}
     hoveredSlateSound {resourceObject=null}
     normal {
      bIsDynamicallyLoaded=false
      drawAs=RoundedBox
      imageType=NoImage
      mirroring=NoMirror
      resourceName=None
      resourceObject=null
      tiling=NoTile
      imageSize=<32 32>
      margin=<ltrb 0 0 0 0>
      outlineSettings {
       bUseBrushTransparency=false
       roundingType=FixedRadius
       width=1
       color {colorUseRule=UseColor_ColorTable, specifiedColor=<rgba 0 0 0 0>}
       cornerRadii=<4 4 4 4>
      }
      tintColor {colorUseRule=UseColor_Specified, specifiedColor=<rgba 0 0.1759 0.8469 1>}
      uVRegion {bIsValid=false, max=<0 0>, min=<0 0>}
     }
     normalForeground {colorUseRule=UseColor_ColorTable, specifiedColor=<rgba 0 0 0 0>}
     normalPadding=<ltrb 0 0 0 0>
     pressed {
      bIsDynamicallyLoaded=false
      drawAs=RoundedBox
      imageType=NoImage
      mirroring=NoMirror
      resourceName=None
      resourceObject=null
      tiling=NoTile
      imageSize=<32 32>
      margin=<ltrb 0 0 0 0>
      outlineSettings {
       bUseBrushTransparency=false
       roundingType=FixedRadius
       width=1
       color {colorUseRule=UseColor_ColorTable, specifiedColor=<rgba 0 0 0 0>}
       cornerRadii=<4 4 4 4>
      }
      tintColor {colorUseRule=UseColor_Specified, specifiedColor=<rgba 0.8469 0.4314 0.2687 1>}
      uVRegion {bIsValid=false, max=<0 0>, min=<0 0>}
     }
     pressedForeground {colorUseRule=UseColor_ColorTable, specifiedColor=<rgba 0 0 0 0>}
     pressedPadding=<ltrb 12 2.5 12 0.5>
     pressedSlateSound {resourceObject=null}
    }}
   slot {
    class=CanvasPanelSlot
    properties {layoutData {alignment=<0 0>, anchors {maximum=<0 0>, minimum=<0 0>}, offsets=<ltrb 1264 632 205.3768 31.3718>}}
   }
  }
  TextBlock_4 {
   widgetClassPath=&TextBlock
   parent=Bake_IK_FK_Tool_Button
   widget=TextBlock_4
   properties {
    text=Bake IK FK Tool
    font {
     bForceMonospaced=false
     bMaterialIsStencil=false
     fontMaterial=null
     fontObject=&Roboto
     letterSpacing=0
     monospacedWidth=1
     size=18
     skewAmount=0
     typefaceFontName=Bold
     outlineSettings {
      bApplyOutlineToDropShadows=false
      bMiteredCorners=false
      bSeparateFillAlpha=false
      outlineMaterial=null
      outlineSize=0
      outlineColor=<rgba 0 0 0 1>
     }
    }
   }
   slot {class=ButtonSlot, properties {padding=<ltrb 0 0 0 0>}}
  }
  DialogSlot {
   widgetClassPath=&NamedSlot
   bIsVariable=true
   parent=CanvasPanel
   widget=DialogSlot
   slot {
    class=CanvasPanelSlot
    properties {layoutData {alignment=<0 0>, anchors {maximum=<0 0>, minimum=<0 0>}, offsets=<ltrb 1152 668 448.8163 287.7747>}}
   }
  }
 ]
}
