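# Editor utility context menu, derived from EUW_ContextMenu, that lists Control Rig sequencer bindings as clickable character items for the animation picker.
# Add Character Item resolves the binding's bound object in the focused Level Sequence and splits its display name into category and tooltip text.
# It adds a black category header once per new name, then an item labeled with the track name, storing the binding and binding the caller's click delegate.
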EditorUtilityWidgetBlueprint /AnimInEdTools/AnimationPicker/EUW_CharacterContextMenu
graphs [
 Add Character Item {nodes [
   n1 = Add Character Item!() -> (then, EventToBind:"Delegate (by ref)", Font Size:double=10, ControlRigBinding:Control Rig Sequencer Binding Proxy Structure) {is=title K2Node_FunctionEntry}
   > SetMaxDesiredHeight!(execute, self:Size Box Object Reference=Get SizeBox() -> (SizeBox:Size Box Object Reference), InMaxDesiredHeight:float=225) -> (then)
   > n2 = GetFocusedLevelSequence!(execute) -> (then, ReturnValue:Level Sequence Object Reference)
   n3 = Break Control Rig Sequencer Binding Proxy!(ControlRigSequencerBindingProxy:"Control Rig Sequencer Binding Proxy Structure (by ref)"=@n1.ControlRigBinding) -> (
    Proxy:Movie Scene Binding Proxy Structure
    ControlRig:Control Rig Object Reference
    Track:Movie Scene Control Rig Parameter Track Object Reference
   ) {is=title K2Node_BreakStruct}
   > n4 = GetBindingID!(execute, Sequence:Movie Scene Sequence Object Reference=@n2, InBinding:"Movie Scene Binding Proxy Structure (by ref)"=@n3) -> (then, ReturnValue:Movie Scene Object Binding ID Structure)
   n5 = GetDisplayName(Object="Get (a copy)"(Array:Array of Object References=GetBoundObjects!(ObjectBinding:Movie Scene Object Binding ID Structure=@n4) -> (ReturnValue:Array of Object References)) -> (Output:Object Reference))
   n6 = Split(SourceString=@n5, InStr=., SearchDir=FromStart)
   n7 = Split(SourceString=@n6.RightS, InStr=" ", SearchDir=FromStart)
   n8 = Select(Option 0:String=@n5, Option 1:String=@n6, Index=@n6.ReturnValue) -> (ReturnValue:String)
   > Set Tooltip(Tooltip:String=Select(Option 0=@n8, Option 1=@n7, Index=@n7.ReturnValue)) -> (Output_Get:String)
   n9 = Get Created Categoriess() -> (Created Categoriess:Array of Strings)
   > n10 = Branch(Condition=OR Boolean(A=IsEmpty(InString=@n8), B=Contains Item(TargetArray:Array of Strings=@n9, ItemToFind:"String (by ref)"=@n8)))
   n11 = Create Widget!(
    execute:Exec=[@n16.then, @n10.then]
    WorldContextObject:Object Reference
    Class:User Widget Class Reference=&/EUW_CharacterContextMenu_Item.EUW_CharacterContextMenu_Item_C
    OwningPlayer:Player Controller Object Reference
   ) -> (then:Exec, ReturnValue:EUW Character Context Menu Item Object Reference)
   > AddChild(self=Get VerticalBox() -> (VerticalBox:Vertical Box Object Reference), Content=@n11)
   > Set Context Menu!(execute, Context Menu:EUW Context Menu Object Reference=Self-Reference(), self:EUW Character Context Menu Item Object Reference=@n11) -> (then, Output_Get:EUW Context Menu Object Reference)
   > Set ControlRig Sequencer Binding!(
    execute
    ControlRig Sequencer Binding:Control Rig Sequencer Binding Proxy Structure=Get ControlRigBinding() -> (ControlRigBinding:Control Rig Sequencer Binding Proxy Structure)
    self:EUW Character Context Menu Item Object Reference=@n11
   ) -> (then, Output_Get:Control Rig Sequencer Binding Proxy Structure)
   > Bind Event to Context Menu Item Clicked(self:EUW Character Context Menu Item Object Reference=@n11, Delegate=Get EventToBind() -> (EventToBind:"Delegate (by ref)")) {is=title K2Node_AddDelegate}
   > n12 = GetDisplayName!(execute:Exec, Track:Movie Scene Track Object Reference=@n3.Track) -> (then:Exec, ReturnValue:Text)
   n13 = Get ButtonText!(self:EUW Character Context Menu Item Object Reference=@n11) -> (ButtonText:Text Object Reference)
   > Set Text(Text=@n12, self=@n13)
   > Set Font!(
    execute
    Font:Slate Font Info Structure=MakeSlateFontInfo!(
     FontObject:Object Reference=&Roboto
     FontMaterial:Object Reference
     OutlineSettings:Font Outline Settings Structure=(OutlineSize=0,bMiteredCorners=False,bSeparateFillAlpha=False,bApplyOutlineToDropShadows=False,OutlineMaterial=None,OutlineColor=(R=0,G=0,B=0,A=1))
     TypefaceFontName:Name=Bold
     Size:float=Get Font Size() -> (Font Size:double)
     LetterSpacing:int=0
     SkewAmount:float=0
     bForceMonospaced:bool=false
     bMaterialIsStencil:bool=false
     MonospacedWidth:float=1
    ) -> (ReturnValue:Slate Font Info Structure)
    self:Text Object Reference=@n13
   ) -> (then, Output_Get:Slate Font Info Structure)
   > Set ToolTipText!(execute, ToolTipText:Text="To Text (String)"(InString=Get Tooltip() -> (Tooltip:String)), self:Widget Object Reference=@n13) -> (then, Output_Get:Text)
   > Return Node!(execute, OutputPin:EUW Character Context Menu Item Object Reference=@n11)
   Add(execute=@n10.else, TargetArray=@n9, NewItem=@n8) {is=title K2Node_CallArrayFunction}
   > n14 = Create Widget(Class=/AnimInEdTools/ContextMenu/EUW_ContextMenu_CategoryItem.EUW_ContextMenu_CategoryItem_C) -> (ReturnValue:EUW Context Menu Category Item Object Reference)
   > AddChild(self=Get VerticalBox(), Content=@n14)
   n15 = Get ButtonText(self:EUW Context Menu Category Item Object Reference=@n14)
   > Set Text(Text="To Text (String)"(InString=@n8), self=@n15)
   > Set Font(
    Font=MakeSlateFontInfo(
     FontObject=&Roboto
     OutlineSettings=(OutlineSize=0,bMiteredCorners=False,bSeparateFillAlpha=False,bApplyOutlineToDropShadows=False,OutlineMaterial=None,OutlineColor=(R=0,G=0,B=0,A=1))
     TypefaceFontName=Bold
     Size=Get Font Size()
     LetterSpacing=0
     SkewAmount=0
     bForceMonospaced=false
     bMaterialIsStencil=false
     MonospacedWidth=1
    )
    self=@n15
   )
   > n16 = SetBrushColor(
    self=Get Border!(self:EUW Context Menu Category Item Object Reference=@n14) -> (Border:Border Object Reference)
    InBrushColor=(R=0,G=0,B=0,A=1)
   )
  ]}
 EventGraph {nodes [
   Event Construct()
   n1 = Make Margin(Left=0, Top=8, Right=0, Bottom=0)
   > Set ScrollbarPadding!(execute, ScrollbarPadding:Margin Structure=@n1, self:Scroll Box Object Reference=Get ScrollBox() -> (ScrollBox:Scroll Box Object Reference)) -> (then, Output_Get:Margin Structure)
   > Set Padding!(execute, Padding:Margin Structure=@n1, self:Border Object Reference=Get Border!() -> (Border)) -> (then, Output_Get:Margin Structure)
   n2 = Event Tick()
   > Parent: Tick!(execute, MyGeometry:Geometry Structure=@n2.MyGeometry, InDeltaTime:float=@n2.InDeltaTime) -> (then) {is=title K2Node_CallParentFunction}
  ]}
]
variables []
widgets {info {parentClass=/AnimInEdTools/ContextMenu/EUW_ContextMenu.EUW_ContextMenu_C, rootWidgetClass=null, widgetCount=0}, widgets []}
