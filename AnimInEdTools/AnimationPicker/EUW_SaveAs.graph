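# Editor Utility Widget that forms the Save As panel for Animation Picker, composing a content path and file name without saving the asset itself.
# The path defaults to the user's Developers folder or existing canvas, or comes from the Content Browser or SaveGame history.
# A context menu selects among supplied Control Rig sequencer bindings.
# A serialize button exports SaveGame locations, pickers and tab bindings to home-directory JSON via inline Python.
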EditorUtilityWidgetBlueprint /AnimInEdTools/AnimationPicker/EUW_SaveAs
graphs [
 EventGraph {nodes [
   n1 = "On Selection Changed (PreviousPaths_ComboBox)"!() -> (OutputDelegate, then, SelectedItem:String, SelectionType:ESelectInfo Enum) {is=title K2Node_ComponentBoundEvent}
   > Set Text(
    Text="To Text (String)"(InString=@n1.SelectedItem)
    self:Text Box Object Reference=Get FilePathTextBox() -> (FilePathTextBox:Editor Utility Editable Text Box Object Reference)
   )
   > Set Full Path() {is=title K2Node_CallFunction}
   "On Clicked (GetContentBrowserFolder_Button)"() {is=title K2Node_ComponentBoundEvent}
   > Set Text(Text="To Text (String)"(InString=GetCurrentContentBrowserPath!() -> (OutPath:String, ReturnValue:bool)), self=Get FilePathTextBox())
   > Set Full Path()
   "On Clicked (Character_ContextMenu_Button)"() {is=title K2Node_ComponentBoundEvent}
   > Clear Context Menu(self:EUW Context Menu Object Reference=Get Character_ContextMenu() -> (Character_ContextMenu:EUW Character Context Menu Object Reference))
   > n2 = For Each Loop(Array:Array of Control Rig Sequencer Binding Proxy Structures=Get ControlRigs() -> (ControlRigs:Array of Control Rig Sequencer Binding Proxy Structures)) -> (Array Element:Control Rig Sequencer Binding Proxy Structure)
   n3 = Get Character_ContextMenu_Button() -> (Character_ContextMenu_Button:Editor Utility Button Object Reference)
   n4 = BreakVector2D(InVec=GetAbsolutePosition!(Geometry:"Geometry Structure (by ref)"=GetCachedGeometry!(self:Widget Object Reference=@n3) -> (ReturnValue:"Geometry Structure (by ref)")) -> (ReturnValue:Vector 2D Structure))
   Show!(
    execute=@n2.Completed
    self:EUW Context Menu Object Reference=Get Character_ContextMenu()
    Show:bool=true
    Mouse Position:Vector 2D Structure=MakeVector2D(X=@n4, Y=float + float(A=@n4.Y, B=100))
    Name:String
   ) -> (then)
   n5 = CustomEvent!() -> (
    OutputDelegate
    then
    Context Item:Editor Utility Widget Object Reference
    Slot Widget:Editor Utility Widget Object Reference
    ControlRig Sequencer BInding:Control Rig Sequencer Binding Proxy Structure
   ) {is=title K2Node_CustomEvent}
   Add Character Item!(
    execute=@n2.LoopBody
    self:EUW Character Context Menu Object Reference=Get Character_ContextMenu()
    EventToBind:"Delegate (by ref)"=@n5
    Font Size:double=10
    ControlRigBinding:Control Rig Sequencer Binding Proxy Structure=@n2
   ) -> (then, OutputPin:EUW Character Context Menu Item Object Reference)
   n6 = "On Text Changed (FilePathTextBox)"!() -> (OutputDelegate, then, Text:"Text (by ref)") {is=title K2Node_ComponentBoundEvent}
   Set Full Path(execute=[@n6.then, @n7.then])
   n7 = "On Text Changed (FileNameTextBox)"!() -> (OutputDelegate, then, Text:"Text (by ref)") {is=title K2Node_ComponentBoundEvent}
   "On Opening (PreviousPaths_ComboBox)"() {is=title K2Node_ComponentBoundEvent}
   > n8 = LoadGameFromSlot!(execute, SlotName:String=Animation Picker, UserIndex:int=0) -> (then, ReturnValue:Save Game Object Reference)
   > n9 = Cast To SG_AnimationPicker!(execute, Object=@n8) -> (then, CastFailed, AsSG Animation Picker:SG Animation Picker Object Reference) {is=title K2Node_DynamicCast}
   > Set FilePaths(FilePaths:Array of Strings=Get Saved Locations!(self:SG Animation Picker Object Reference=@n9) -> (Saved Locations:Array of Strings)) -> (Output_Get:Array of Strings)
   > ClearOptions(self:"Combo Box (String) Object Reference"=Get PreviousPaths_ComboBox() -> (PreviousPaths_ComboBox:Editor Utility Combo Box String Object Reference))
   > n10 = Reverse for Each Loop!(Exec:Exec, Array:Array of Strings=Get FilePaths() -> (FilePaths:Array of Strings)) -> (LoopBody:Exec, ArrayIndex:int, ArrayElement:String, Completed:Exec) {is=title K2Node_MacroInstance}
   > AddOption!(execute, self:"Combo Box (String) Object Reference"=Get PreviousPaths_ComboBox(), Option:String=@n10.ArrayElement) -> (then)
   "On Clicked (SerializeSaveGameObject_Button)"() {is=title K2Node_ComponentBoundEvent}
   > n11 = LoadGameFromSlot(SlotName=Animation Picker, UserIndex=0)
   > n12 = Cast To SG_AnimationPicker(Object=@n11)
   > Execute Python Script(PythonScript="import os\r\nimport json\r\n\r\n_SOURCE_FOLDER = os.path.expanduser(\"~\") + '/'\r\n_ANIM_PICKER_JSON = 'AnimationPickerUserSettings.json'\r\n\r\nfile_path = _SOURCE_FOLDER + _ANIM_PICKER_JSON\r\n\r\n## Always overwrite with new data\r\ndata = {\"SavedLocations\": []}\r\n\r\nwith open(file_path, 'w') as file:\r\n    json.dump(data, file, indent=4)\r\n\r\nprint(\"Serialized Save Game Object : \" + _SOURCE_FOLDER+_ANIM_PICKER_JSON)")
   > n13 = For Each Loop(Array:Array of Strings=Get Saved Locations(self=@n12)) -> (Array Element:String)
   > Execute Python Script!(
    execute
    PythonScript="import os\r\nimport json\r\n\r\n_SOURCE_FOLDER = os.path.expanduser(\"~\") + '/'\r\n_ANIM_PICKER_JSON = 'AnimationPickerUserSettings.json'\r\n\r\nfile_path = _SOURCE_FOLDER + _ANIM_PICKER_JSON\r\n\r\nif os.path.isfile(file_path):\r\n    try:\r\n        with open(file_path, 'r') as file:\r\n            data = json.load(file)\r\n    except (json.JSONDecodeError, ValueError):\r\n        data = {\"SavedLocations\": []}\r\nelse:\r\n    data = {\"SavedLocations\": []}\r\n\r\ndata[\"SavedLocations\"].append(saved_location)\r\n\r\nwith open(file_path, 'w') as file:\r\n    json.dump(data, file, indent=4)"
    saved_location:String=@n13
   ) -> (then, ReturnValue) {inputs [saved_location]}
   > Set MapIndex(MapIndex:int=0) -> (Output_Get:int=0)
   > n14 = "For Each Loop (Map)"!(
    execute:Exec
    MapPin:Map of Control Rig Asset Soft Reference Structures to DA Animation Picker Canvas Object References=Get User Pickers!(self:SG Animation Picker Object Reference=@n12) -> (User Pickers:Map of Control Rig Asset Soft Reference Structures to DA Animation Picker Canvas Object References)
    BreakPin:Exec
   ) -> (then:Exec, KeyPin:Control Rig Asset Soft Reference Structure, ValuePin:DA Animation Picker Canvas Object Reference, CompletedPin:Exec)
   > Execute Python Script!(
    execute
    PythonScript="import os\r\nimport json\r\n\r\n_SOURCE_FOLDER = os.path.expanduser(\"~\") + '/'\r\n_ANIM_PICKER_JSON = 'AnimationPickerUserSettings.json'\r\n\r\nfile_path = _SOURCE_FOLDER+_ANIM_PICKER_JSON\r\n\r\nif os.path.isfile(file_path):\r\n    with open(file_path, 'r') as file:\r\n        data = json.load(file)\r\n\r\n        if \"UserPickers\" not in data:\r\n            data[\"UserPickers\"] = {}\r\n\r\n        data[\"UserPickers\"][map_index] = [control_soft_class, canvas_path]\r\n\r\n        ## Save updated data\r\n        with open(file_path, 'w') as file:\r\n            json.dump(data, file, indent=4)"
    map_index:int=Get MapIndex() -> (MapIndex:int=0)
    control_soft_class:String="To String (Control Rig Soft Reference)"!(InRef:"Control Rig Asset Soft Reference Structure (by ref)"=@n14) -> (ReturnValue:String)
    canvas_path:String=Get Object Path String(Object=@n14.ValuePin)
   ) -> (then, ReturnValue) {inputs [map_index, control_soft_class, canvas_path]}
   > Set MapIndex(MapIndex=int + int(A=Get MapIndex() -> (MapIndex=0))) -> (Output_Get=0)
   n15 = "For Each Loop (Map)"(
    execute=@n14.CompletedPin
    MapPin:Map of Strings to STRUCT Picker Tab Data Structures=Get Tabs ControlRig Binding!(self:SG Animation Picker Object Reference=@n12) -> (Tabs ControlRig Binding:Map of Strings to STRUCT Picker Tab Data Structures)
   ) -> (KeyPin:String, ValuePin:STRUCT Picker Tab Data Structure)
   n16 = Break STRUCT Picker Tab Data!(STRUCT_PickerTabData:"STRUCT Picker Tab Data Structure (by ref)"=@n15.ValuePin) -> (
    ControlRigBinding_13_#1:Control Rig Sequencer Binding Proxy Structure
    Index_16_#2:int
    Bookmarks_17_#3:Map of Strings to Widget Transform Structures
   ) {is=title K2Node_BreakStruct}
   > Execute Python Script!(
    execute
    PythonScript="import os\r\nimport json\r\n\r\n_SOURCE_FOLDER = os.path.expanduser(\"~\") + '/'\r\n_ANIM_PICKER_JSON = 'AnimationPickerUserSettings.json'\r\n\r\nfile_path = _SOURCE_FOLDER+_ANIM_PICKER_JSON\r\n\r\nif os.path.isfile(file_path):\r\n    with open(file_path, 'r') as file:\r\n        data = json.load(file)\r\n\r\n        if \"TabControlRig\" not in data:\r\n            data[\"TabControlRig\"] = {}\r\n\r\n        data[\"TabControlRig\"][tab_name] = control_rig_name\r\n\r\n        ## Save updated data\r\n        with open(file_path, 'w') as file:\r\n            json.dump(data, file, indent=4)"
    tab_name:String=@n15
    control_rig_name:String=GetDisplayName(Object=Break Control Rig Sequencer Binding Proxy!(ControlRigSequencerBindingProxy:"Control Rig Sequencer Binding Proxy Structure (by ref)"=@n16) -> (
      Proxy:Movie Scene Binding Proxy Structure
      ControlRig:Control Rig Object Reference
      Track:Movie Scene Control Rig Parameter Track Object Reference
     ) {is=title K2Node_BreakStruct}.ControlRig)
   ) -> (then, ReturnValue) {inputs [tab_name, control_rig_name]}
   Set Full Path() {is=title K2Node_CustomEvent}
   n17 = Get Text(self=Get FilePathTextBox())
   n18 = RemoveDuplicateSlashes!(InPath:String=Append(A="To String (Text)"(InText=@n17), B=/)) -> (OutPath:String)
   n19 = "To String (Text)"(InText=Get Text(self=Get FileNameTextBox() -> (FileNameTextBox:Editor Utility Editable Text Box Object Reference)))
   > Set Text(
    Text="To Text (String)"(InString=Append(A=@n18, B=@n19))
    self:Text Object Reference=Get FullPathTextBlock() -> (FullPathTextBlock:Text Object Reference)
   )
   Event Construct()
   > SetVisibility(self=Get Character_ContextMenu(), InVisibility=Collapsed)
   > Set Text(self=Get ErrorTextBlock() -> (ErrorTextBlock:Text Object Reference))
   > Set Text(self=Get FullPathTextBlock())
   n20 = Split(SourceString=Split(SourceString=GameUserDeveloperDir!() -> (ReturnValue:String), InStr=/, SearchDir=FromEnd), InStr=/, SearchDir=FromEnd)
   > Set Text(
    Text="To Text (String)"(InString=Append!(A=/Game/Developers/, B=@n20.RightS, C:String=/AnimationPicker/Character/) -> (ReturnValue))
    self:Text Box Object Reference=Get FilePathTextBox()
   )
   > n21 = Branch(Condition=IsValid(Object=Get Existing User Canvas() -> (Existing User Canvas:DA Animation Picker Canvas Object Reference)))
   n22 = Get Existing User Canvas()
   n23 = Split(SourceString=Get Object Path String(Object=@n22), InStr=/, SearchDir=FromEnd)
   > Set Text(Text="To Text (String)"(InString=Append!(A=@n23, B=/) -> (ReturnValue)), self=Get FilePathTextBox())
   > n24 = Set Text(Text="To Text (String)"(InString=GetDisplayName(Object=@n22)), self=Get FileNameTextBox())
   Set Full Path(execute=[@n21.else, @n24.then]) {is=title K2Node_CallFunction}
   n25 = Get ControlRigs()
   > n26 = Branch(Condition="integer > integer"(A=Length(TargetArray:Array of Control Rig Sequencer Binding Proxy Structures=@n25)))
   > Set Visibility(Visibility=Visible, self=Get ControlRig_HorizontalBox() -> (ControlRig_HorizontalBox:Horizontal Box Object Reference))
   Set Visibility(execute=@n26.else, Visibility=Collapsed, self=Get ControlRig_HorizontalBox())
   PrintString(
    execute=@n27.then
    InString=GetDisplayName(Object=Break Control Rig Sequencer Binding Proxy(ControlRigSequencerBindingProxy=@n5.ControlRig Sequencer BInding).ControlRig)
   )
   Event PreConstruct()
   Event Tick()
   Set Character TextBlock!(
    execute=@n5.then
    self
    ControlRigSequencerBindingProxy:"Control Rig Sequencer Binding Proxy Structure (by ref)"=@n5.ControlRig Sequencer BInding
   ) -> (then) {is=title K2Node_CallFunction}
   > n27 = Set Control Rig Sequencer BInding(Control Rig Sequencer BInding:Control Rig Sequencer Binding Proxy Structure=@n5.ControlRig Sequencer BInding) -> (Output_Get:Control Rig Sequencer Binding Proxy Structure)
  ]}
 Set Character TextBlock {nodes [
   n1 = Set Character TextBlock!() -> (then, ControlRigSequencerBindingProxy:"Control Rig Sequencer Binding Proxy Structure (by ref)") {is=title K2Node_FunctionEntry}
   > n2 = GetFocusedLevelSequence!(execute) -> (then, ReturnValue:Level Sequence Object Reference)
   n3 = Break Control Rig Sequencer Binding Proxy(ControlRigSequencerBindingProxy=@n1)
   > n4 = GetBindingID!(execute, Sequence:Movie Scene Sequence Object Reference=@n2, InBinding:"Movie Scene Binding Proxy Structure (by ref)"=@n3) -> (then, ReturnValue:Movie Scene Object Binding ID Structure)
   n5 = GetBoundObjects!(ObjectBinding:Movie Scene Object Binding ID Structure=@n4) -> (ReturnValue:Array of Object References)
   > Branch(Condition="integer > integer"(A=Length(TargetArray:Array of Object References=@n5)))
   > Is Valid(InputObject=@n3.Track)
   > n6 = GetDisplayName!(execute:Exec, Track:Movie Scene Track Object Reference=@n3.Track) -> (then:Exec, ReturnValue:Text)
   > Set Text(
    Text="To Text (String)"(InString="To String (Text)"(InText=@n6))
    self:Text Object Reference=Get RigName_TextBlock() -> (RigName_TextBlock:Text Object Reference)
   )
   n7 = GetDisplayName!(Object:Object Reference="Get (a copy)"(Array:Array of Object References=@n5) -> (Output:Object Reference)) -> (ReturnValue:String)
   n8 = Split(SourceString=@n7, InStr=., SearchDir=FromStart)
   n9 = Select(Option 0:String=@n7, Option 1:String=@n8, Index=@n8.ReturnValue) -> (ReturnValue:String)
   > Set Text(Text="To Text (String)"(InString=@n9), self=Get CharacterName_TextBlock() -> (CharacterName_TextBlock:Text Object Reference))
   n10 = Split(SourceString=@n8.RightS, InStr=" ", SearchDir=FromStart)
   > Set ToolTipText!(
    execute
    ToolTipText:Text="To Text (String)"(InString=Select(Option 0=@n9, Option 1=@n10, Index=@n10.ReturnValue))
    self:Widget Object Reference=Get Character_ContextMenu_Button()
   ) -> (then, Output_Get:Text)
  ]}
]
variables [
 Canvas File {type=DA_AnimationPicker_Canvas_C, default=null}
 Control Rig Sequencer BInding {
  type=ControlRigSequencerBindingProxy
  default {controlRig=null, track=null, proxy {bindingId=#4, sequence=null}}
 }
 ControlRigs {type="array<ControlRigSequencerBindingProxy>", default []}
 Existing User Canvas {type=DA_AnimationPicker_Canvas_C, default=null}
 FilePath {type=string, default=""}
 FilePaths {type="array<string>", default []}
 MapIndex {type=integer}
]
widgets {
 info {rootWidgetClass=&VerticalBox, widgetCount=26}
 widgets [
  VerticalBox_0 {widgetClassPath=&VerticalBox, widget=VerticalBox_0}
  Overlay_293 {widgetClassPath=&Overlay, parent=VerticalBox_0, widget=Overlay_293, slot {class=VerticalBoxSlot}}
  VerticalBox {
   widgetClassPath=&VerticalBox
   parent=Overlay_293
   widget=VerticalBox
   slot {class=OverlaySlot, properties {horizontalAlignment=HAlign_Fill, verticalAlignment=VAlign_Fill}}
  }
  ControlRig_HorizontalBox {
   widgetClassPath=&HorizontalBox
   bIsVariable=true
   parent=VerticalBox
   widget=ControlRig_HorizontalBox
   slot {class=VerticalBoxSlot, properties {padding=<ltrb 2 2 2 2>}}
  }
  TextBlock_1 {
   widgetClassPath=&TextBlock
   parent=ControlRig_HorizontalBox
   widget=TextBlock_1
   properties {
    text="ControlRig :  "
    colorAndOpacity {colorUseRule=UseColor_Specified, specifiedColor=<rgba 0.5271 0.5271 0.5271 1>}
    font {
     bForceMonospaced=false
     bMaterialIsStencil=false
     fontMaterial=null
     fontObject=&Roboto
     letterSpacing=0
     monospacedWidth=1
     size=8
     skewAmount=0
     typefaceFontName=Regular
     outlineSettings {
      bApplyOutlineToDropShadows=false
      bMiteredCorners=false
      bSeparateFillAlpha=false
      outlineMaterial=null
      outlineSize=0
      outlineColor=<rgba 0 0 0 1>
     }
    }
   }
   slot {class=HorizontalBoxSlot, properties {horizontalAlignment=HAlign_Left, verticalAlignment=VAlign_Center, padding=<ltrb 2 0 2 0>}}
  }
  SizeBox_2 {
   widgetClassPath=&SizeBox
   parent=ControlRig_HorizontalBox
   widget=SizeBox_2
   properties {bOverride_MinDesiredHeight=true, bOverride_MinDesiredWidth=true, minDesiredHeight=38, minDesiredWidth=200}
   slot {class=HorizontalBoxSlot, properties {horizontalAlignment=HAlign_Left}}
  }
  Character_ContextMenu_Button {
   widgetClassPath=&EditorUtilityButton
   parent=SizeBox_2
   widget=Character_ContextMenu_Button
   properties {widgetStyle {
     clickedSlateSound {resourceObject=null}
     disabled {
      bIsDynamicallyLoaded=false
      drawAs=RoundedBox
      imageType=NoImage
      mirroring=NoMirror
      resourceName=None
      resourceObject=null
      tiling=NoTile
      imageSize=<0 0>
      margin=<ltrb 0 0 0 0>
      outlineSettings {
       bUseBrushTransparency=false
       roundingType=FixedRadius
       width=1
       color {colorUseRule=UseColor_ColorTable, specifiedColor=<rgba 0 0 0 0>}
       cornerRadii=<4 4 4 4>
      }
      tintColor {colorUseRule=UseColor_Specified, specifiedColor=<rgba 0.0395 0.0395 0.0395 1>}
      uVRegion {bIsValid=false, max=<0 0>, min=<0 0>}
     }
     disabledForeground {colorUseRule=UseColor_ColorTable, specifiedColor=<rgba 0 0 0 0>}
     hovered {
      bIsDynamicallyLoaded=false
      drawAs=RoundedBox
      imageType=NoImage
      mirroring=NoMirror
      resourceName=None
      resourceObject=null
      tiling=NoTile
      imageSize=<32 32>
      margin=<ltrb 0 0 0 0>
      outlineSettings {
       bUseBrushTransparency=false
       roundingType=FixedRadius
       width=1
       color {colorUseRule=UseColor_ColorTable, specifiedColor=<rgba 0 0 0 0>}
       cornerRadii=<4 4 4 4>
      }
      tintColor {colorUseRule=UseColor_Specified, specifiedColor=<rgba 0 0 0 1>}
      uVRegion {bIsValid=false, max=<0 0>, min=<0 0>}
     }
     hoveredForeground {colorUseRule=UseColor_ColorTable, specifiedColor=<rgba 0 0 0 0>}
     hoveredSlateSound {resourceObject=null}
     normal {
      bIsDynamicallyLoaded=false
      drawAs=RoundedBox
      imageType=NoImage
      mirroring=NoMirror
      resourceName=None
      resourceObject=null
      tiling=NoTile
      imageSize=<32 32>
      margin=<ltrb 0 0 0 0>
      outlineSettings {
       bUseBrushTransparency=false
       roundingType=FixedRadius
       width=1
       color {colorUseRule=UseColor_ColorTable, specifiedColor=<rgba 0 0 0 0>}
       cornerRadii=<4 4 4 4>
      }
      tintColor {colorUseRule=UseColor_Specified, specifiedColor=<rgba 0 0 0 1>}
      uVRegion {bIsValid=false, max=<0 0>, min=<0 0>}
     }
     normalForeground {colorUseRule=UseColor_ColorTable, specifiedColor=<rgba 0 0 0 0>}
     normalPadding=<ltrb 12 1.5 12 1.5>
     pressed {
      bIsDynamicallyLoaded=false
      drawAs=RoundedBox
      imageType=NoImage
      mirroring=NoMirror
      resourceName=None
      resourceObject=null
      tiling=NoTile
      imageSize=<32 32>
      margin=<ltrb 0 0 0 0>
      outlineSettings {
       bUseBrushTransparency=false
       roundingType=FixedRadius
       width=1
       color {colorUseRule=UseColor_ColorTable, specifiedColor=<rgba 0 0 0 0>}
       cornerRadii=<4 4 4 4>
      }
      tintColor {colorUseRule=UseColor_Specified, specifiedColor=<rgba 0.0284 0.0284 0.0284 1>}
      uVRegion {bIsValid=false, max=<0 0>, min=<0 0>}
     }
     pressedForeground {colorUseRule=UseColor_ColorTable, specifiedColor=<rgba 0 0 0 0>}
     pressedPadding=<ltrb 12 2.5 12 0.5>
     pressedSlateSound {resourceObject=null}
    }}
   slot {class=SizeBoxSlot}
  }
  VerticalBox_51 {
   widgetClassPath=&VerticalBox
   parent=Character_ContextMenu_Button
   widget=VerticalBox_51
   slot {class=ButtonSlot, properties {horizontalAlignment=HAlign_Fill, verticalAlignment=VAlign_Fill, padding=<ltrb 0 0 0 0>}}
  }
  CharacterName_TextBlock {
   widgetClassPath=&TextBlock
   bIsVariable=true
   parent=VerticalBox_51
   widget=CharacterName_TextBlock
   properties {
    text=Select a Control Rig
    colorAndOpacity {colorUseRule=UseColor_Specified, specifiedColor=<rgba 0.2683 0.2683 0.2683 1>}
    font {
     bForceMonospaced=false
     bMaterialIsStencil=false
     fontMaterial=null
     fontObject=&Roboto
     letterSpacing=0
     monospacedWidth=1
     size=9.75
     skewAmount=0
     typefaceFontName=Bold
     outlineSettings {
      bApplyOutlineToDropShadows=false
      bMiteredCorners=false
      bSeparateFillAlpha=false
      outlineMaterial=null
      outlineSize=0
      outlineColor=<rgba 0 0 0 1>
     }
    }
   }
   slot {class=VerticalBoxSlot, properties {horizontalAlignment=HAlign_Left, verticalAlignment=VAlign_Center, padding=<ltrb 0 2 0 0>}}
  }
  RigName_TextBlock {
   widgetClassPath=&TextBlock
   bIsVariable=true
   parent=VerticalBox_51
   widget=RigName_TextBlock
   properties {font {
     bForceMonospaced=false
     bMaterialIsStencil=false
     fontMaterial=null
     fontObject=&Roboto
     letterSpacing=0
     monospacedWidth=1
     size=9.75
     skewAmount=0
     typefaceFontName=Bold
     outlineSettings {
      bApplyOutlineToDropShadows=false
      bMiteredCorners=false
      bSeparateFillAlpha=false
      outlineMaterial=null
      outlineSize=0
      outlineColor=<rgba 0 0 0 1>
     }
    }}
   slot {class=VerticalBoxSlot, properties {horizontalAlignment=HAlign_Left, verticalAlignment=VAlign_Center, padding=<ltrb 14 2 4 2>}}
  }
  HorizontalBox {
   widgetClassPath=&HorizontalBox
   parent=VerticalBox
   widget=HorizontalBox
   slot {class=VerticalBoxSlot, properties {padding=<ltrb 2 2 2 2>}}
  }
  TextBlock {
   widgetClassPath=&TextBlock
   parent=HorizontalBox
   widget=TextBlock
   properties {
    text="File Path :  "
    colorAndOpacity {colorUseRule=UseColor_Specified, specifiedColor=<rgba 0.5271 0.5271 0.5271 1>}
    font {
     bForceMonospaced=false
     bMaterialIsStencil=false
     fontMaterial=null
     fontObject=&Roboto
     letterSpacing=0
     monospacedWidth=1
     size=8
     skewAmount=0
     typefaceFontName=Regular
     outlineSettings {
      bApplyOutlineToDropShadows=false
      bMiteredCorners=false
      bSeparateFillAlpha=false
      outlineMaterial=null
      outlineSize=0
      outlineColor=<rgba 0 0 0 1>
     }
    }
   }
   slot {class=HorizontalBoxSlot, properties {horizontalAlignment=HAlign_Left, verticalAlignment=VAlign_Center, padding=<ltrb 2 0 2 0>}}
  }
  GetContentBrowserFolder_Button {
   widgetClassPath=&EditorUtilityButton
   parent=HorizontalBox
   widget=GetContentBrowserFolder_Button
   properties {toolTipText=Sets the File Path to the currently selected folder in the Content Browser}
   slot {class=HorizontalBoxSlot}
  }
  TextBlock_57 {
   widgetClassPath=&TextBlock
   parent=GetContentBrowserFolder_Button
   widget=TextBlock_57
   properties {
    justification=Center
    text=">"
    font {
     bForceMonospaced=false
     bMaterialIsStencil=false
     fontMaterial=null
     fontObject=&Roboto
     letterSpacing=0
     monospacedWidth=1
     size=7.5
     skewAmount=0
     typefaceFontName=Regular
     outlineSettings {
      bApplyOutlineToDropShadows=false
      bMiteredCorners=false
      bSeparateFillAlpha=false
      outlineMaterial=null
      outlineSize=0
      outlineColor=<rgba 0 0 0 1>
     }
    }
   }
   slot {class=ButtonSlot, properties {padding=<ltrb 0 0 0 0>}}
  }
  FilePathTextBox {
   widgetClassPath=&EditorUtilityEditableTextBox
   bIsVariable=true
   parent=HorizontalBox
   widget=FilePathTextBox
   slot {class=HorizontalBoxSlot, properties {size {sizeRule=Fill, value=1}}}
  }
  Border_1 {
   widgetClassPath=&Border
   parent=HorizontalBox
   widget=Border_1
   properties {brushColor=<rgba 1 1 1 0>, desiredSizeScale=<0.5 1>, padding=<ltrb 0 0 0 0>}
   slot {class=HorizontalBoxSlot, properties {padding=<ltrb 3 0 0 0>, size {sizeRule=Fill, value=0.1}}}
  }
  PreviousPaths_ComboBox {
   widgetClassPath=/Script/Blutility.EditorUtilityComboBoxString
   bIsVariable=true
   parent=Border_1
   widget=PreviousPaths_ComboBox
   properties {
    toolTipText=List of Previously Used File Paths
    font {
     bForceMonospaced=false
     bMaterialIsStencil=false
     fontMaterial=null
     fontObject=&Roboto
     letterSpacing=0
     monospacedWidth=1
     size=9
     skewAmount=0
     typefaceFontName=Regular
     outlineSettings {
      bApplyOutlineToDropShadows=false
      bMiteredCorners=false
      bSeparateFillAlpha=false
      outlineMaterial=null
      outlineSize=0
      outlineColor=<rgba 0 0 0 1>
     }
    }
    foregroundColor {colorUseRule=UseColor_Specified, specifiedColor=<rgba 0.5271 0.5271 0.5271 1>}
    widgetStyle {
     comboButtonStyle {
      downArrowAlign=VAlign_Center
      buttonStyle {
       clickedSlateSound {resourceObject=null}
       disabled {
        bIsDynamicallyLoaded=false
        drawAs=RoundedBox
        imageType=NoImage
        mirroring=NoMirror
        resourceName=None
        resourceObject=null
        tiling=NoTile
        imageSize=<0 0>
        margin=<ltrb 0 0 0 0>
        outlineSettings {
         bUseBrushTransparency=false
         roundingType=FixedRadius
         width=1
         color {colorUseRule=UseColor_ColorTable, specifiedColor=<rgba 0 0 0 0>}
         cornerRadii=<4 4 4 4>
        }
        tintColor {colorUseRule=UseColor_Specified, specifiedColor=<rgba 0.0395 0.0395 0.0395 1>}
        uVRegion {bIsValid=false, max=<0 0>, min=<0 0>}
       }
       disabledForeground {colorUseRule=UseColor_ColorTable, specifiedColor=<rgba 0 0 0 0>}
       hovered {
        bIsDynamicallyLoaded=false
        drawAs=RoundedBox
        imageType=NoImage
        mirroring=NoMirror
        resourceName=None
        resourceObject=null
        tiling=NoTile
        imageSize=<0 0>
        margin=<ltrb 0 0 0 0>
        outlineSettings {
         bUseBrushTransparency=false
         roundingType=FixedRadius
         width=1
         color {colorUseRule=UseColor_ColorTable, specifiedColor=<rgba 0 0 0 0>}
         cornerRadii=<4 4 4 4>
        }
        tintColor {colorUseRule=UseColor_Specified, specifiedColor=<rgba 0.0953 0.0953 0.0953 1>}
        uVRegion {bIsValid=false, max=<0 0>, min=<0 0>}
       }
       hoveredForeground {colorUseRule=UseColor_ColorTable, specifiedColor=<rgba 0 0 0 0>}
       hoveredSlateSound {resourceObject=null}
       normal {
        bIsDynamicallyLoaded=false
        drawAs=RoundedBox
        imageType=NoImage
        mirroring=NoMirror
        resourceName=None
        resourceObject=null
        tiling=NoTile
        imageSize=<0 0>
        margin=<ltrb 0 0 0 0>
        outlineSettings {
         bUseBrushTransparency=false
         roundingType=FixedRadius
         width=1
         color {colorUseRule=UseColor_ColorTable, specifiedColor=<rgba 0 0 0 0>}
         cornerRadii=<4 4 4 4>
        }
        tintColor {colorUseRule=UseColor_Specified, specifiedColor=<rgba 0.0395 0.0395 0.0395 1>}
        uVRegion {bIsValid=false, max=<0 0>, min=<0 0>}
       }
       normalForeground {colorUseRule=UseColor_ColorTable, specifiedColor=<rgba 0 0 0 0>}
       normalPadding=<ltrb 8 1 8 1>
       pressed {
        bIsDynamicallyLoaded=false
        drawAs=RoundedBox
        imageType=NoImage
        mirroring=NoMirror
        resourceName=None
        resourceObject=null
        tiling=NoTile
        imageSize=<0 0>
        margin=<ltrb 0 0 0 0>
        outlineSettings {
         bUseBrushTransparency=false
         roundingType=FixedRadius
         width=1
         color {colorUseRule=UseColor_ColorTable, specifiedColor=<rgba 0 0 0 0>}
         cornerRadii=<4 4 4 4>
        }
        tintColor {colorUseRule=UseColor_Specified, specifiedColor=<rgba 0.0284 0.0284 0.0284 1>}
        uVRegion {bIsValid=false, max=<0 0>, min=<0 0>}
       }
       pressedForeground {colorUseRule=UseColor_ColorTable, specifiedColor=<rgba 0 0 0 0>}
       pressedPadding=<ltrb 8 2 8 0>
       pressedSlateSound {resourceObject=null}
      }
      contentPadding=<ltrb 0 0 0 0>
      downArrowImage {
       bIsDynamicallyLoaded=false
       drawAs=NoDrawType
       imageType=Vector
       mirroring=NoMirror
       resourceName=../../../Engine/Content/Slate/Starship/CoreWidgets/ComboBox/wide-chevron-down.svg
       resourceObject=null
       tiling=NoTile
       imageSize=<16 16>
       margin=<ltrb 0 0 0 0>
       outlineSettings {
        bUseBrushTransparency=false
        roundingType=HalfHeightRadius
        width=0
        color {colorUseRule=UseColor_Specified, specifiedColor=<rgba 0 0 0 0>}
        cornerRadii=<0 0 0 0>
       }
       tintColor {colorUseRule=UseColor_Specified, specifiedColor=<rgba 1 1 1 1>}
       uVRegion {bIsValid=false, max=<0 0>, min=<0 0>}
      }
      downArrowPadding=<ltrb 2 2 2 2>
      menuBorderBrush {
       bIsDynamicallyLoaded=false
       drawAs=RoundedBox
       imageType=NoImage
       mirroring=NoMirror
       resourceName=None
       resourceObject=null
       tiling=NoTile
       imageSize=<0 0>
       margin=<ltrb 0 0 0 0>
       outlineSettings {
        bUseBrushTransparency=false
        roundingType=FixedRadius
        width=1
        color {colorUseRule=UseColor_ColorTable, specifiedColor=<rgba 0 0 0 0>}
        cornerRadii=<0 0 0 0>
       }
       tintColor {colorUseRule=UseColor_Specified, specifiedColor=<rgba 0.0395 0.0395 0.0395 1>}
       uVRegion {bIsValid=false, max=<0 0>, min=<0 0>}
      }
      menuBorderPadding=<ltrb 0 0 0 0>
      shadowColorAndOpacity=<rgba 0 0 0 1>
      shadowOffset=<0 0>
     }
     contentPadding=<ltrb 0 0 0 0>
     menuRowPadding=<ltrb 8 3 8 3>
     pressedSlateSound {resourceObject=null}
     selectionChangeSlateSound {resourceObject=null}
    }
   }
   slot {class=BorderSlot, properties {padding=<ltrb 0 0 0 0>}}
  }
  HorizontalBox_0 {
   widgetClassPath=&HorizontalBox
   parent=VerticalBox
   widget=HorizontalBox_0
   slot {class=VerticalBoxSlot, properties {padding=<ltrb 2 2 2 2>}}
  }
  TextBlock_3 {
   widgetClassPath=&TextBlock
   parent=HorizontalBox_0
   widget=TextBlock_3
   properties {
    text=File Name :
    colorAndOpacity {colorUseRule=UseColor_Specified, specifiedColor=<rgba 0.5271 0.5271 0.5271 1>}
    font {
     bForceMonospaced=false
     bMaterialIsStencil=false
     fontMaterial=null
     fontObject=&Roboto
     letterSpacing=0
     monospacedWidth=1
     size=8
     skewAmount=0
     typefaceFontName=Regular
     outlineSettings {
      bApplyOutlineToDropShadows=false
      bMiteredCorners=false
      bSeparateFillAlpha=false
      outlineMaterial=null
      outlineSize=0
      outlineColor=<rgba 0 0 0 1>
     }
    }
   }
   slot {class=HorizontalBoxSlot, properties {horizontalAlignment=HAlign_Left, verticalAlignment=VAlign_Center, padding=<ltrb 2 0 2 0>}}
  }
  FileNameTextBox {
   widgetClassPath=&EditorUtilityEditableTextBox
   bIsVariable=true
   parent=HorizontalBox_0
   widget=FileNameTextBox
   slot {class=HorizontalBoxSlot, properties {size {sizeRule=Fill, value=1}}}
  }
  HorizontalBox_1 {
   widgetClassPath=&HorizontalBox
   parent=VerticalBox
   widget=HorizontalBox_1
   slot {class=VerticalBoxSlot, properties {verticalAlignment=VAlign_Top, padding=<ltrb 2 0 2 0>}}
  }
  FullPathTextBlock {
   widgetClassPath=&TextBlock
   bIsVariable=true
   parent=HorizontalBox_1
   widget=FullPathTextBlock
   properties {
    text=my/file/path
    colorAndOpacity {colorUseRule=UseColor_Specified, specifiedColor=<rgba 0.5271 0.5271 0.5271 1>}
    font {
     bForceMonospaced=false
     bMaterialIsStencil=false
     fontMaterial=null
     fontObject=&Roboto
     letterSpacing=0
     monospacedWidth=1
     size=8
     skewAmount=0
     typefaceFontName=Regular
     outlineSettings {
      bApplyOutlineToDropShadows=false
      bMiteredCorners=false
      bSeparateFillAlpha=false
      outlineMaterial=null
      outlineSize=0
      outlineColor=<rgba 0 0 0 1>
     }
    }
   }
   slot {class=HorizontalBoxSlot}
  }
  ErrorTextBlock {
   widgetClassPath=&TextBlock
   bIsVariable=true
   parent=HorizontalBox_1
   widget=ErrorTextBlock
   properties {
    text=Error
    colorAndOpacity {colorUseRule=UseColor_Specified, specifiedColor=<rgba 0.5271 0.0357 0.0357 1>}
    font {
     bForceMonospaced=false
     bMaterialIsStencil=false
     fontMaterial=null
     fontObject=&Roboto
     letterSpacing=0
     monospacedWidth=1
     size=8
     skewAmount=0
     typefaceFontName=Bold
     outlineSettings {
      bApplyOutlineToDropShadows=false
      bMiteredCorners=false
      bSeparateFillAlpha=false
      outlineMaterial=null
      outlineSize=0
      outlineColor=<rgba 0 0 0 1>
     }
    }
   }
   slot {class=HorizontalBoxSlot, properties {padding=<ltrb 5 0 0 0>}}
  }
  Spacer_75 {
   widgetClassPath=&Spacer
   parent=HorizontalBox_1
   widget=Spacer_75
   slot {class=HorizontalBoxSlot, properties {size {sizeRule=Fill, value=1}}}
  }
  SerializeSaveGameObject_Button {
   widgetClassPath=&EditorUtilityButton
   parent=HorizontalBox_1
   widget=SerializeSaveGameObject_Button
   properties {
    toolTipText=Serialize Save Game Object
    widgetStyle {
     clickedSlateSound {resourceObject=null}
     disabled {
      bIsDynamicallyLoaded=false
      drawAs=RoundedBox
      imageType=NoImage
      mirroring=NoMirror
      resourceName=None
      resourceObject=null
      tiling=NoTile
      imageSize=<0 0>
      margin=<ltrb 0 0 0 0>
      outlineSettings {
       bUseBrushTransparency=false
       roundingType=FixedRadius
       width=1
       color {colorUseRule=UseColor_ColorTable, specifiedColor=<rgba 0 0 0 0>}
       cornerRadii=<4 4 4 4>
      }
      tintColor {colorUseRule=UseColor_Specified, specifiedColor=<rgba 0.0395 0.0395 0.0395 1>}
      uVRegion {bIsValid=false, max=<0 0>, min=<0 0>}
     }
     disabledForeground {colorUseRule=UseColor_ColorTable, specifiedColor=<rgba 0 0 0 0>}
     hovered {
      bIsDynamicallyLoaded=false
      drawAs=RoundedBox
      imageType=FullColor
      mirroring=NoMirror
      resourceName=None
      resourceObject=/Game/UI/SocialPanel/Art/Textures/T_UI_Icon_LocalUserSettings
      tiling=NoTile
      imageSize=<16 16>
      margin=<ltrb 0 0 0 0>
      outlineSettings {
       bUseBrushTransparency=false
       roundingType=FixedRadius
       width=1
       color {colorUseRule=UseColor_ColorTable, specifiedColor=<rgba 0 0 0 0>}
       cornerRadii=<4 4 4 4>
      }
      tintColor {colorUseRule=UseColor_Specified, specifiedColor=<rgba 0.0953 0.0953 0.0953 1>}
      uVRegion {bIsValid=false, max=<0 0>, min=<0 0>}
     }
     hoveredForeground {colorUseRule=UseColor_ColorTable, specifiedColor=<rgba 0 0 0 0>}
     hoveredSlateSound {resourceObject=null}
     normal {
      bIsDynamicallyLoaded=false
      drawAs=RoundedBox
      imageType=FullColor
      mirroring=NoMirror
      resourceName=None
      resourceObject=&T_UI_Icon_LocalUserSettings
      tiling=NoTile
      imageSize=<16 16>
      margin=<ltrb 0 0 0 0>
      outlineSettings {
       bUseBrushTransparency=false
       roundingType=FixedRadius
       width=1
       color {colorUseRule=UseColor_ColorTable, specifiedColor=<rgba 0 0 0 0>}
       cornerRadii=<4 4 4 4>
      }
      tintColor {colorUseRule=UseColor_Specified, specifiedColor=<rgba 0.4137 0.4069 0.3498 1>}
      uVRegion {bIsValid=false, max=<0 0>, min=<0 0>}
     }
     normalForeground {colorUseRule=UseColor_ColorTable, specifiedColor=<rgba 0 0 0 0>}
     normalPadding=<ltrb 12 1.5 12 1.5>
     pressed {
      bIsDynamicallyLoaded=false
      drawAs=RoundedBox
      imageType=FullColor
      mirroring=NoMirror
      resourceName=None
      resourceObject=&T_UI_Icon_LocalUserSettings
      tiling=NoTile
      imageSize=<16 16>
      margin=<ltrb 0 0 0 0>
      outlineSettings {
       bUseBrushTransparency=false
       roundingType=FixedRadius
       width=1
       color {colorUseRule=UseColor_ColorTable, specifiedColor=<rgba 0 0 0 0>}
       cornerRadii=<4 4 4 4>
      }
      tintColor {colorUseRule=UseColor_Specified, specifiedColor=<rgba 0.0284 0.0284 0.0284 1>}
      uVRegion {bIsValid=false, max=<0 0>, min=<0 0>}
     }
     pressedForeground {colorUseRule=UseColor_ColorTable, specifiedColor=<rgba 0 0 0 0>}
     pressedPadding=<ltrb 12 2.5 12 0.5>
     pressedSlateSound {resourceObject=null}
    }
   }
   slot {class=HorizontalBoxSlot, properties {horizontalAlignment=HAlign_Right}}
  }
  Character_ContextMenu {
   widgetClassPath=&/EUW_CharacterContextMenu
   bIsVariable=true
   parent=Overlay_293
   widget=Character_ContextMenu
   slot {class=OverlaySlot}
  }
 ]
}
