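# Editor utility widget for an animation scanner tool that checks selected assets or folders for montage, blend space, proxy table or named-curve usage.
# Construct reloads the AnimationScanner Python module and passes it the buttons, results list, filters, edit panel and scan options.
# Criteria checkboxes and the curve-name field forward changes to that module through Execute Python Script nodes, so the Blueprint holds only layout and glue.
# A collapsed preview box supplies sample fonts and a browse button.
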EditorUtilityWidgetBlueprint /AnimInEdTools/AnimationScanner/EUW_AnimationScanner
graphs [EventGraph {nodes [
   n1 = "On Check State Changed (Checkbox_IsUsedInMontage?)"!() -> (OutputDelegate, then, bIsChecked:bool) {is=title K2Node_ComponentBoundEvent}
   > Execute Python Script!(
    execute
    PythonScript="import sys\r\nAnimationScanner = sys.modules['AnimationScanner']\r\nAnimationScanner.cls.set_is_used_in_montage(is_checked)"
    is_checked:bool=@n1.bIsChecked
   ) -> (then, ReturnValue) {inputs [is_checked]}
   n2 = "On Check State Changed (Checkbox_DoesContainAnimCurve?)"!() -> (OutputDelegate, then, bIsChecked:bool) {is=title K2Node_ComponentBoundEvent}
   > Execute Python Script(
    PythonScript="import sys\r\nAnimationScanner = sys.modules['AnimationScanner']\r\nAnimationScanner.cls.set_does_contain_animation_curve(is_checked)"
    is_checked=@n2.bIsChecked
   ) {inputs [is_checked]}
   n3 = "On Check State Changed (Checkbox_IsUsedInBlendSpace?)"!() -> (OutputDelegate, then, bIsChecked:bool) {is=title K2Node_ComponentBoundEvent}
   > Execute Python Script(
    PythonScript="import sys\r\nAnimationScanner = sys.modules['AnimationScanner']\r\nAnimationScanner.cls.set_is_used_in_blend_space(is_checked)"
    is_checked=@n3.bIsChecked
   ) {inputs [is_checked]}
   n4 = "On Check State Changed (Checkbox_IsUsedInPT?)"!() -> (OutputDelegate, then, bIsChecked:bool) {is=title K2Node_ComponentBoundEvent}
   > Execute Python Script(
    PythonScript="import sys\r\nAnimationScanner = sys.modules['AnimationScanner']\r\nAnimationScanner.cls.set_is_used_in_proxy_table(is_checked)"
    is_checked=@n4.bIsChecked
   ) {inputs [is_checked]}
   "OnTextChanged (Editable Text) (EditableText_CurveNames)"!() -> (OutputDelegate, then, Text:"Text (by ref)") {is=title K2Node_ComponentBoundEvent}
   > Execute Python Script!(
    execute
    PythonScript="import sys\r\nAnimationScanner = sys.modules['AnimationScanner']\r\nAnimationScanner.cls.set_does_contain_animation_curve_text_color()"
   ) -> (then, ReturnValue)
   Event Construct()
   > Execute Python Script(PythonScript="import AnimationScanner\r\nimport importlib\r\nimportlib.reload(AnimationScanner)")
   > Execute Python Script!(
    execute
    PythonScript="import sys\r\nAnimationScanner = sys.modules['AnimationScanner']\r\n\r\nAnimationScanner.cls.init_widget_input_anim_curves(widget_input_anim_curves)\r\nAnimationScanner.cls.init_widget_scan_button(widget_scan_button)\r\nAnimationScanner.cls.init_widget_scan_results_layout(widget_results_layout, widget_results_scroll_box)\r\nAnimationScanner.cls.init_apply_changes_button(widget_apply_button)\r\nAnimationScanner.cls.init_fonts(header_font, item_font)\r\nAnimationScanner.cls.init_widget_filter_combo_box(widget_filter_match_type, widget_filter_object_type)\r\nAnimationScanner.cls.init_widget_browse_button(widget_browse_button)"
    widget_input_anim_curves:Editor Utility Editable Text Object Reference=Get EditableText_CurveNames() -> (EditableText_CurveNames:Editor Utility Editable Text Object Reference)
    widget_scan_button:Editor Utility Button Object Reference=Get Button_Scan() -> (Button_Scan:Editor Utility Button Object Reference)
    widget_apply_button:Editor Utility Button Object Reference=Get Button_Apply() -> (Button_Apply:Editor Utility Button Object Reference)
    widget_results_scroll_box:Editor Utility Scroll Box Object Reference=Get ScrollBox_Results() -> (ScrollBox_Results:Editor Utility Scroll Box Object Reference)
    header_font:Text Object Reference=Get TextBlock_Sample_header() -> (TextBlock_Sample_header:Text Object Reference)
    item_font:Text Object Reference=Get TextBlock_Sample_item() -> (TextBlock_Sample_item:Text Object Reference)
    widget_results_layout:Vertical Box Object Reference=Get VerticalBox_ScanResults() -> (VerticalBox_ScanResults:Vertical Box Object Reference)
    widget_filter_match_type:Editor Utility Combo Box String Object Reference=Get EditorUtilityComboBoxString_FilterMatch() -> (EditorUtilityComboBoxString_FilterMatch:Editor Utility Combo Box String Object Reference)
    widget_filter_object_type:Editor Utility Combo Box String Object Reference=Get EditorUtilityComboBoxString_FilterObject() -> (EditorUtilityComboBoxString_FilterObject:Editor Utility Combo Box String Object Reference)
    widget_browse_button:Editor Utility Button Object Reference=Get EditorUtilityButton_Sample() -> (EditorUtilityButton_Sample:Editor Utility Button Object Reference)
   ) -> (then, ReturnValue) {inputs [
     widget_input_anim_curves
     widget_scan_button
     widget_apply_button
     widget_results_scroll_box
     header_font
     item_font
     widget_results_layout
     widget_filter_match_type
     widget_filter_object_type
     widget_browse_button
    ]}
   > Execute Python Script!(
    execute
    PythonScript="import sys\r\nAnimationScanner = sys.modules['AnimationScanner']\r\nAnimationScanner.cls.init_widget_edit_v_box(widget, widget2, widget3)"
    widget:Vertical Box Object Reference=Get VerticalBox_Edit() -> (VerticalBox_Edit:Vertical Box Object Reference)
    widget2:Vertical Box Object Reference=Get VerticalBox_EditCurves() -> (VerticalBox_EditCurves:Vertical Box Object Reference)
    widget3:Editor Utility Expandable Area Object Reference=Get EditorUtilityExpandableArea_EditCurves() -> (EditorUtilityExpandableArea_EditCurves:Editor Utility Expandable Area Object Reference)
   ) -> (then, ReturnValue) {inputs [widget,widget2,widget3]}
   > Execute Python Script!(
    execute
    PythonScript="import sys\r\nAnimationScanner = sys.modules['AnimationScanner']\r\nAnimationScanner.cls.init_widget_options(widget_checkbox_option_items,\r\nwidget_checkbox_option_folders, widget_checkbox_option_project)"
    widget_checkbox_option_items:Editor Utility Check Box Object Reference=Get EditorUtilityCheckBox_Items() -> (EditorUtilityCheckBox_Items:Editor Utility Check Box Object Reference)
    widget_checkbox_option_folders:Editor Utility Check Box Object Reference=Get EditorUtilityCheckBox_Folders() -> (EditorUtilityCheckBox_Folders:Editor Utility Check Box Object Reference)
    widget_checkbox_option_project:Editor Utility Check Box Object Reference=Get EditorUtilityCheckBox_Project() -> (EditorUtilityCheckBox_Project:Editor Utility Check Box Object Reference)
   ) -> (then, ReturnValue) {inputs [widget_checkbox_option_items, widget_checkbox_option_folders, widget_checkbox_option_project]}
   Event PreConstruct()
   Event Tick()
   Event Destruct()
   Execute Python Script!(execute, PythonScript="import sys\r\ndel sys.modules['AnimationScanner']") -> (then, ReturnValue)
   Execute Python Script!(
    execute
    PythonScript="import sys\r\nAnimationScanner = sys.modules['AnimationScanner']\r\nAnimationScanner.cls.set_debug_widget(debug_widget)"
    debug_widget:Wildcard
   ) -> (then, ReturnValue) {inputs [debug_widget]}
  ]}]
variables []
widgets {
 info {namedSlotCount=4, rootWidgetClass=&EditorUtilityScrollBox, widgetCount=62}
 widgets [
  EditorUtilityScrollBox_1 {widgetClassPath=&EditorUtilityScrollBox, widget=EditorUtilityScrollBox_1}
  VerticalBox_0 {
   widgetClassPath=&VerticalBox
   parent=EditorUtilityScrollBox_1
   widget=VerticalBox_0
   properties {renderTransformPivot=<0 0>}
   slot {class=ScrollBoxSlot}
  }
  Border_117 {
   widgetClassPath=&Border
   parent=VerticalBox_0
   widget=Border_117
   properties {
    background {
     bIsDynamicallyLoaded=false
     drawAs=Border
     imageType=NoImage
     mirroring=NoMirror
     resourceName=None
     resourceObject=null
     tiling=NoTile
     imageSize=<32 32>
     margin=<ltrb 0 0 0 1>
     outlineSettings {
      bUseBrushTransparency=false
      roundingType=HalfHeightRadius
      width=0
      color {colorUseRule=UseColor_Specified, specifiedColor=<rgba 0 0 0 0>}
      cornerRadii=<0 0 0 0>
     }
     tintColor {colorUseRule=UseColor_Specified, specifiedColor=<rgba 0.1302 0.1302 0.1302 1>}
     uVRegion {bIsValid=false, max=<0 0>, min=<0 0>}
    }
    padding=<ltrb 6 10 0 0>
   }
   slot {class=VerticalBoxSlot}
  }
  TextBlock_351 {
   widgetClassPath=&TextBlock
   parent=Border_117
   widget=TextBlock_351
   properties {
    text=Scan :
    toolTipText="Using Checkboxed select one or several criterias to scan for.\r\n"
    font {
     bForceMonospaced=false
     bMaterialIsStencil=false
     fontMaterial=null
     fontObject=null
     letterSpacing=0
     monospacedWidth=1
     size=9
     skewAmount=0
     typefaceFontName=RLSmall-Bold
     outlineSettings {
      bApplyOutlineToDropShadows=false
      bMiteredCorners=false
      bSeparateFillAlpha=false
      outlineMaterial=null
      outlineSize=0
      outlineColor=<rgba 0 0 0 1>
     }
    }
   }
   slot {class=BorderSlot, properties {padding=<ltrb 6 10 0 0>}}
  }
  Spacer {
   widgetClassPath=&Spacer
   parent=VerticalBox_0
   widget=Spacer
   properties {size=<1 10>}
   slot {class=VerticalBoxSlot, properties {horizontalAlignment=HAlign_Left, verticalAlignment=VAlign_Center}}
  }
  HorizontalBoxMontage {
   widgetClassPath=&HorizontalBox
   parent=VerticalBox_0
   widget=HorizontalBoxMontage
   slot {class=VerticalBoxSlot, properties {horizontalAlignment=HAlign_Left, verticalAlignment=VAlign_Center}}
  }
  Checkbox_IsUsedInMontage? {
   widgetClassPath=&EditorUtilityCheckBox
   bIsVariable=true
   parent=HorizontalBoxMontage
   widget=Checkbox_IsUsedInMontage?
   properties {horizontalAlignment=HAlign_Left, toolTipText=Is animation used in any Montage Asset}
   slot {class=HorizontalBoxSlot, properties {horizontalAlignment=HAlign_Left, verticalAlignment=VAlign_Center, padding=<ltrb 30 0 10 0>}}
  }
  TextBlock_64 {
   widgetClassPath=&TextBlock
   parent=HorizontalBoxMontage
   widget=TextBlock_64
   properties {
    text=Is in Montage ?
    toolTipText=Is animation used in any Montage Asset
    font {
     bForceMonospaced=false
     bMaterialIsStencil=false
     fontMaterial=null
     fontObject=null
     letterSpacing=0
     monospacedWidth=1
     size=9
     skewAmount=0
     typefaceFontName=RLSmall-Regular
     outlineSettings {
      bApplyOutlineToDropShadows=false
      bMiteredCorners=false
      bSeparateFillAlpha=false
      outlineMaterial=null
      outlineSize=0
      outlineColor=<rgba 0 0 0 1>
     }
    }
   }
   slot {class=HorizontalBoxSlot, properties {horizontalAlignment=HAlign_Left, verticalAlignment=VAlign_Center}}
  }
  HorizontalBoxBlendSpace {
   widgetClassPath=&HorizontalBox
   parent=VerticalBox_0
   widget=HorizontalBoxBlendSpace
   slot {class=VerticalBoxSlot, properties {horizontalAlignment=HAlign_Left, verticalAlignment=VAlign_Center}}
  }
  Checkbox_IsUsedInBlendSpace? {
   widgetClassPath=&EditorUtilityCheckBox
   bIsVariable=true
   parent=HorizontalBoxBlendSpace
   widget=Checkbox_IsUsedInBlendSpace?
   properties {horizontalAlignment=HAlign_Left, toolTipText=Is animation used in any BlendSpace Asset}
   slot {class=HorizontalBoxSlot, properties {horizontalAlignment=HAlign_Left, verticalAlignment=VAlign_Center, padding=<ltrb 30 0 10 0>}}
  }
  TextBlock_2 {
   widgetClassPath=&TextBlock
   parent=HorizontalBoxBlendSpace
   widget=TextBlock_2
   properties {
    text=Is in BlendSpace ?
    toolTipText=Is animation used in any BlendSpace Asset
    font {
     bForceMonospaced=false
     bMaterialIsStencil=false
     fontMaterial=null
     fontObject=null
     letterSpacing=0
     monospacedWidth=1
     size=9
     skewAmount=0
     typefaceFontName=RLSmall-Regular
     outlineSettings {
      bApplyOutlineToDropShadows=false
      bMiteredCorners=false
      bSeparateFillAlpha=false
      outlineMaterial=null
      outlineSize=0
      outlineColor=<rgba 0 0 0 1>
     }
    }
   }
   slot {class=HorizontalBoxSlot, properties {horizontalAlignment=HAlign_Left, verticalAlignment=VAlign_Center}}
  }
  HorizontalBoxBlendSpace_1 {
   widgetClassPath=&HorizontalBox
   parent=VerticalBox_0
   widget=HorizontalBoxBlendSpace_1
   slot {class=VerticalBoxSlot, properties {horizontalAlignment=HAlign_Left, verticalAlignment=VAlign_Center}}
  }
  Checkbox_IsUsedInPT? {
   widgetClassPath=&EditorUtilityCheckBox
   bIsVariable=true
   parent=HorizontalBoxBlendSpace_1
   widget=Checkbox_IsUsedInPT?
   properties {horizontalAlignment=HAlign_Left, toolTipText=Is animation used in any ProxyTable Asset}
   slot {class=HorizontalBoxSlot, properties {horizontalAlignment=HAlign_Left, verticalAlignment=VAlign_Center, padding=<ltrb 30 0 10 0>}}
  }
  TextBlock_3 {
   widgetClassPath=&TextBlock
   parent=HorizontalBoxBlendSpace_1
   widget=TextBlock_3
   properties {
    text=Is in ProxyTable ?
    toolTipText=Is animation used in any ProxyTable Asset
    font {
     bForceMonospaced=false
     bMaterialIsStencil=false
     fontMaterial=null
     fontObject=null
     letterSpacing=0
     monospacedWidth=1
     size=9
     skewAmount=0
     typefaceFontName=RLSmall-Regular
     outlineSettings {
      bApplyOutlineToDropShadows=false
      bMiteredCorners=false
      bSeparateFillAlpha=false
      outlineMaterial=null
      outlineSize=0
      outlineColor=<rgba 0 0 0 1>
     }
    }
   }
   slot {class=HorizontalBoxSlot, properties {horizontalAlignment=HAlign_Left, verticalAlignment=VAlign_Center}}
  }
  VerticalBox_551 {
   widgetClassPath=&VerticalBox
   parent=VerticalBox_0
   widget=VerticalBox_551
   slot {class=VerticalBoxSlot, properties {horizontalAlignment=HAlign_Left, verticalAlignment=VAlign_Center}}
  }
  HorizontalBoxCurves {widgetClassPath=&HorizontalBox, parent=VerticalBox_551, widget=HorizontalBoxCurves, slot {class=VerticalBoxSlot}}
  Checkbox_DoesContainAnimCurve? {
   widgetClassPath=&EditorUtilityCheckBox
   bIsVariable=true
   parent=HorizontalBoxCurves
   widget=Checkbox_DoesContainAnimCurve?
   properties {horizontalAlignment=HAlign_Left, toolTipText="Does animation contain float channel curves.\r\n\r\n\r\nType name in the filed on the right."}
   slot {class=HorizontalBoxSlot, properties {horizontalAlignment=HAlign_Left, verticalAlignment=VAlign_Center, padding=<ltrb 30 0 10 0>}}
  }
  TextBlock {
   widgetClassPath=&TextBlock
   parent=HorizontalBoxCurves
   widget=TextBlock
   properties {
    text="Has Curve(s):"
    toolTipText="Does animation contain float channel curves.\r\n\r\n\r\nType name in the filed on the right."
    font {
     bForceMonospaced=false
     bMaterialIsStencil=false
     fontMaterial=null
     fontObject=null
     letterSpacing=0
     monospacedWidth=1
     size=9
     skewAmount=0
     typefaceFontName=RLSmall-Regular
     outlineSettings {
      bApplyOutlineToDropShadows=false
      bMiteredCorners=false
      bSeparateFillAlpha=false
      outlineMaterial=null
      outlineSize=0
      outlineColor=<rgba 0 0 0 1>
     }
    }
   }
   slot {class=HorizontalBoxSlot, properties {horizontalAlignment=HAlign_Left, verticalAlignment=VAlign_Center}}
  }
  EditableText_CurveNames {
   widgetClassPath=/Script/Blutility.EditorUtilityEditableText
   bIsVariable=true
   parent=HorizontalBoxCurves
   widget=EditableText_CurveNames
   properties {
    hintText="[  CURVE NAME(S)  ]"
    toolTipText="Type curve(s) name in this field. \r\nMultiple curve names should be separated by commas"
    widgetStyle {
     backgroundImageComposing {
      bIsDynamicallyLoaded=false
      drawAs=Border
      imageType=FullColor
      mirroring=NoMirror
      resourceName=../../../Engine/Content/Slate/Old/HyperlinkDotted.png
      resourceObject=null
      tiling=Both
      imageSize=<0 0>
      margin=<ltrb 0 0 0 0.1875>
      outlineSettings {
       bUseBrushTransparency=false
       roundingType=HalfHeightRadius
       width=0
       color {colorUseRule=UseColor_Specified, specifiedColor=<rgba 0 0 0 0>}
       cornerRadii=<0 0 0 0>
      }
      tintColor {colorUseRule=UseColor_Specified, specifiedColor=<rgba 1 1 1 1>}
      uVRegion {bIsValid=false, max=<0 0>, min=<0 0>}
     }
     backgroundImageSelected {
      bIsDynamicallyLoaded=false
      drawAs=Image
      imageType=NoImage
      mirroring=NoMirror
      resourceName=None
      resourceObject=null
      tiling=NoTile
      imageSize=<0 0>
      margin=<ltrb 0 0 0 0>
      outlineSettings {
       bUseBrushTransparency=false
       roundingType=HalfHeightRadius
       width=0
       color {colorUseRule=UseColor_Specified, specifiedColor=<rgba 0 0 0 0>}
       cornerRadii=<0 0 0 0>
      }
      tintColor {colorUseRule=UseColor_Specified, specifiedColor=<rgba 0 0.162 0.7454 1>}
      uVRegion {bIsValid=false, max=<0 0>, min=<0 0>}
     }
     caretImage {
      bIsDynamicallyLoaded=false
      drawAs=Image
      imageType=NoImage
      mirroring=NoMirror
      resourceName=None
      resourceObject=null
      tiling=NoTile
      imageSize=<0 0>
      margin=<ltrb 0 0 0 0>
      outlineSettings {
       bUseBrushTransparency=false
       roundingType=HalfHeightRadius
       width=0
       color {colorUseRule=UseColor_Specified, specifiedColor=<rgba 0 0 0 0>}
       cornerRadii=<0 0 0 0>
      }
      tintColor {colorUseRule=UseColor_Specified, specifiedColor=<rgba 1 1 1 1>}
      uVRegion {bIsValid=false, max=<0 0>, min=<0 0>}
     }
     colorAndOpacity {colorUseRule=UseColor_Specified, specifiedColor=<rgba 1 1 1 1>}
     font {
      bForceMonospaced=false
      bMaterialIsStencil=false
      fontMaterial=null
      fontObject=null
      letterSpacing=10
      monospacedWidth=1
      size=9
      skewAmount=0
      typefaceFontName=RLSmall-Italic
      outlineSettings {
       bApplyOutlineToDropShadows=false
       bMiteredCorners=false
       bSeparateFillAlpha=false
       outlineMaterial=null
       outlineSize=0
       outlineColor=<rgba 0 0 0 1>
      }
     }
     hintColorAndOpacity {colorUseRule=UseColor_Foreground, specifiedColor=<rgba 1 0 1 1>}
     selectedBackgroundColor {colorUseRule=UseColor_Foreground, specifiedColor=<rgba 1 0 1 1>}
    }
   }
   slot {class=HorizontalBoxSlot, properties {verticalAlignment=VAlign_Center, padding=<ltrb 10 0 10 0>, size {sizeRule=Fill, value=1}}}
  }
  Text_QuestionMark {
   widgetClassPath=&TextBlock
   parent=HorizontalBoxCurves
   widget=Text_QuestionMark
   properties {
    text=?
    toolTipText="Does animation contain float channel curves.\r\n\r\n\r\nType name in the filed on the right."
    font {
     bForceMonospaced=false
     bMaterialIsStencil=false
     fontMaterial=null
     fontObject=null
     letterSpacing=0
     monospacedWidth=1
     size=9
     skewAmount=0
     typefaceFontName=RLSmall-Regular
     outlineSettings {
      bApplyOutlineToDropShadows=false
      bMiteredCorners=false
      bSeparateFillAlpha=false
      outlineMaterial=null
      outlineSize=0
      outlineColor=<rgba 0 0 0 1>
     }
    }
   }
   slot {class=HorizontalBoxSlot, properties {horizontalAlignment=HAlign_Left, verticalAlignment=VAlign_Center}}
  }
  Spacer_1 {
   widgetClassPath=&Spacer
   parent=VerticalBox_0
   widget=Spacer_1
   properties {size=<1 40>}
   slot {class=VerticalBoxSlot, properties {horizontalAlignment=HAlign_Left, verticalAlignment=VAlign_Center}}
  }
  VerticalBox_550 {
   widgetClassPath=&VerticalBox
   parent=VerticalBox_0
   widget=VerticalBox_550
   slot {class=VerticalBoxSlot, properties {horizontalAlignment=HAlign_Left, verticalAlignment=VAlign_Center, padding=<ltrb 80 0 50 0>}}
  }
  Button_Scan {
   widgetClassPath=&EditorUtilityButton
   parent=VerticalBox_550
   widget=Button_Scan
   properties {toolTipText="Scanns asset(s) currently selected in Content Drawer"}
   slot {class=VerticalBoxSlot}
  }
  TextBlock_1 {
   widgetClassPath=&TextBlock
   parent=Button_Scan
   widget=TextBlock_1
   properties {
    text="Scan[Selected]"
    font {
     bForceMonospaced=false
     bMaterialIsStencil=false
     fontMaterial=null
     fontObject=null
     letterSpacing=0
     monospacedWidth=1
     size=10.5
     skewAmount=0
     typefaceFontName=RLSmall-Bold
     outlineSettings {
      bApplyOutlineToDropShadows=false
      bMiteredCorners=false
      bSeparateFillAlpha=false
      outlineMaterial=null
      outlineSize=0
      outlineColor=<rgba 0 0 0 1>
     }
    }
   }
   slot {class=ButtonSlot, properties {horizontalAlignment=HAlign_Fill, verticalAlignment=VAlign_Fill, padding=<ltrb 20 6 20 6>}}
  }
  EditorUtilityExpandableArea_314 {
   widgetClassPath=/Script/Blutility.EditorUtilityExpandableArea
   bIsVariable=true
   parent=VerticalBox_550
   widget=EditorUtilityExpandableArea_314
   slot {class=VerticalBoxSlot, properties {horizontalAlignment=HAlign_Center, padding=<ltrb 0 0 8 0>}}
  }
  TextBlock_7 {
   widgetClassPath=&TextBlock
   namedSlotHost=EditorUtilityExpandableArea_314
   widget=TextBlock_7
   properties {
    text=Options
    toolTipText=Scan options
    font {
     bForceMonospaced=false
     bMaterialIsStencil=false
     fontMaterial=null
     fontObject=null
     letterSpacing=0
     monospacedWidth=1
     size=9
     skewAmount=0
     typefaceFontName=RLSmall-Regular
     outlineSettings {
      bApplyOutlineToDropShadows=false
      bMiteredCorners=false
      bSeparateFillAlpha=false
      outlineMaterial=null
      outlineSize=0
      outlineColor=<rgba 0 0 0 1>
     }
    }
   }
  }
  VerticalBox_1 {widgetClassPath=&VerticalBox, namedSlotHost=EditorUtilityExpandableArea_314, widget=VerticalBox_1}
  HorizontalBox_0 {
   widgetClassPath=&HorizontalBox
   parent=VerticalBox_1
   widget=HorizontalBox_0
   slot {class=VerticalBoxSlot, properties {padding=<ltrb -14 0 0 0>}}
  }
  EditorUtilityCheckBox_Items {
   widgetClassPath=&EditorUtilityCheckBox
   bIsVariable=true
   parent=HorizontalBox_0
   widget=EditorUtilityCheckBox_Items
   properties {checkedState=Checked, toolTipText=Scan ASSETS selected in the rigth panel of Content Browser}
   slot {class=HorizontalBoxSlot}
  }
  TextBlock_9 {
   widgetClassPath=&TextBlock
   parent=HorizontalBox_0
   widget=TextBlock_9
   properties {
    text=Selected Assets
    toolTipText=Scan ASSETS selected in the rigth panel of Content Browser
    font {
     bForceMonospaced=false
     bMaterialIsStencil=false
     fontMaterial=null
     fontObject=null
     letterSpacing=0
     monospacedWidth=1
     size=9
     skewAmount=0
     typefaceFontName=RLSmall-Regular
     outlineSettings {
      bApplyOutlineToDropShadows=false
      bMiteredCorners=false
      bSeparateFillAlpha=false
      outlineMaterial=null
      outlineSize=0
      outlineColor=<rgba 0 0 0 1>
     }
    }
   }
   slot {class=HorizontalBoxSlot, properties {horizontalAlignment=HAlign_Left, verticalAlignment=VAlign_Center, padding=<ltrb 2 0 0 0>}}
  }
  HorizontalBox {
   widgetClassPath=&HorizontalBox
   parent=VerticalBox_1
   widget=HorizontalBox
   slot {class=VerticalBoxSlot, properties {padding=<ltrb -14 0 0 0>}}
  }
  EditorUtilityCheckBox_Folders {
   widgetClassPath=&EditorUtilityCheckBox
   bIsVariable=true
   parent=HorizontalBox
   widget=EditorUtilityCheckBox_Folders
   properties {toolTipText=Scan anything in FOLDERS selected in the left or right panel of Content Browser}
   slot {class=HorizontalBoxSlot}
  }
  TextBlock_10 {
   widgetClassPath=&TextBlock
   parent=HorizontalBox
   widget=TextBlock_10
   properties {
    text=Selected Folders
    toolTipText=Scan anything in FOLDERS selected in the left or right panel of Content Browser
    font {
     bForceMonospaced=false
     bMaterialIsStencil=false
     fontMaterial=null
     fontObject=null
     letterSpacing=0
     monospacedWidth=1
     size=9
     skewAmount=0
     typefaceFontName=RLSmall-Regular
     outlineSettings {
      bApplyOutlineToDropShadows=false
      bMiteredCorners=false
      bSeparateFillAlpha=false
      outlineMaterial=null
      outlineSize=0
      outlineColor=<rgba 0 0 0 1>
     }
    }
   }
   slot {class=HorizontalBoxSlot, properties {horizontalAlignment=HAlign_Left, verticalAlignment=VAlign_Center, padding=<ltrb 2 0 0 0>}}
  }
  HorizontalBox_1 {
   widgetClassPath=&HorizontalBox
   parent=VerticalBox_1
   widget=HorizontalBox_1
   properties {bIsEnabled=false, toolTipText="Scan anything in FOLDERS selected in the Tool Project Tree View. \r\n\r\nDisabled. Work IN Progress."}
   slot {class=VerticalBoxSlot, properties {padding=<ltrb -14 0 0 0>}}
  }
  EditorUtilityCheckBox_Project {
   widgetClassPath=&EditorUtilityCheckBox
   bIsVariable=true
   parent=HorizontalBox_1
   widget=EditorUtilityCheckBox_Project
   slot {class=HorizontalBoxSlot}
  }
  TextBlock_11 {
   widgetClassPath=&TextBlock
   parent=HorizontalBox_1
   widget=TextBlock_11
   properties {
    text=Project Folders
    font {
     bForceMonospaced=false
     bMaterialIsStencil=false
     fontMaterial=null
     fontObject=null
     letterSpacing=0
     monospacedWidth=1
     size=9
     skewAmount=0
     typefaceFontName=RLSmall-Regular
     outlineSettings {
      bApplyOutlineToDropShadows=false
      bMiteredCorners=false
      bSeparateFillAlpha=false
      outlineMaterial=null
      outlineSize=0
      outlineColor=<rgba 0 0 0 1>
     }
    }
   }
   slot {class=HorizontalBoxSlot, properties {horizontalAlignment=HAlign_Left, verticalAlignment=VAlign_Center, padding=<ltrb 2 0 0 0>}}
  }
  Spacer_2 {
   widgetClassPath=&Spacer
   parent=VerticalBox_0
   widget=Spacer_2
   properties {size=<1 10>}
   slot {class=VerticalBoxSlot, properties {horizontalAlignment=HAlign_Left, verticalAlignment=VAlign_Center}}
  }
  VerticalBox_ScanResults {
   widgetClassPath=&VerticalBox
   bIsVariable=true
   parent=VerticalBox_0
   widget=VerticalBox_ScanResults
   slot {class=VerticalBoxSlot}
  }
  Border_1 {
   widgetClassPath=&Border
   parent=VerticalBox_ScanResults
   widget=Border_1
   properties {
    background {
     bIsDynamicallyLoaded=false
     drawAs=Border
     imageType=NoImage
     mirroring=NoMirror
     resourceName=None
     resourceObject=null
     tiling=NoTile
     imageSize=<32 32>
     margin=<ltrb 0 0 0 1>
     outlineSettings {
      bUseBrushTransparency=false
      roundingType=HalfHeightRadius
      width=0
      color {colorUseRule=UseColor_Specified, specifiedColor=<rgba 0 0 0 0>}
      cornerRadii=<0 0 0 0>
     }
     tintColor {colorUseRule=UseColor_Specified, specifiedColor=<rgba 0.1302 0.1302 0.1302 1>}
     uVRegion {bIsValid=false, max=<0 0>, min=<0 0>}
    }
    padding=<ltrb 6 10 0 0>
   }
   slot {class=VerticalBoxSlot}
  }
  TextBlock_8 {
   widgetClassPath=&TextBlock
   parent=Border_1
   widget=TextBlock_8
   properties {
    text=Scan Results :
    toolTipText="Using Checkboxed select one or several criterias to scan for.\r\n"
    font {
     bForceMonospaced=false
     bMaterialIsStencil=false
     fontMaterial=null
     fontObject=null
     letterSpacing=0
     monospacedWidth=1
     size=9
     skewAmount=0
     typefaceFontName=RLSmall-Bold
     outlineSettings {
      bApplyOutlineToDropShadows=false
      bMiteredCorners=false
      bSeparateFillAlpha=false
      outlineMaterial=null
      outlineSize=0
      outlineColor=<rgba 0 0 0 1>
     }
    }
   }
   slot {class=BorderSlot, properties {padding=<ltrb 6 10 0 0>}}
  }
  HorizontalBox_879 {
   widgetClassPath=&HorizontalBox
   parent=VerticalBox_ScanResults
   widget=HorizontalBox_879
   slot {class=VerticalBoxSlot, properties {horizontalAlignment=HAlign_Right, verticalAlignment=VAlign_Top, padding=<ltrb 6 1 0 0>}}
  }
  EditorUtilityComboBoxString_FilterMatch {
   widgetClassPath=/Script/Blutility.EditorUtilityComboBoxString
   bIsVariable=true
   parent=HorizontalBox_879
   widget=EditorUtilityComboBoxString_FilterMatch
   properties {
    selectedOption=Full match
    defaultOptions [Full match, Partial match, "Full && Partial match", No match, All]
    font {
     bForceMonospaced=false
     bMaterialIsStencil=false
     fontMaterial=null
     fontObject=null
     letterSpacing=0
     monospacedWidth=1
     size=9
     skewAmount=0
     typefaceFontName=RLSmall-Regular
     outlineSettings {
      bApplyOutlineToDropShadows=false
      bMiteredCorners=false
      bSeparateFillAlpha=false
      outlineMaterial=null
      outlineSize=0
      outlineColor=<rgba 0 0 0 1>
     }
    }
    foregroundColor {colorUseRule=UseColor_Specified, specifiedColor=<rgba 0.5271 0.5271 0.5271 1>}
   }
   slot {class=HorizontalBoxSlot}
  }
  EditorUtilityComboBoxString_FilterObject {
   widgetClassPath=&EditorUtilityComboBoxString
   bIsVariable=true
   parent=HorizontalBox_879
   widget=EditorUtilityComboBoxString_FilterObject
   properties {
    selectedOption="Objects && Referencers"
    defaultOptions [Only Objects, Only Referencers, "Objects && Referencers"]
    font {
     bForceMonospaced=false
     bMaterialIsStencil=false
     fontMaterial=null
     fontObject=null
     letterSpacing=0
     monospacedWidth=1
     size=9
     skewAmount=0
     typefaceFontName=RLSmall-Regular
     outlineSettings {
      bApplyOutlineToDropShadows=false
      bMiteredCorners=false
      bSeparateFillAlpha=false
      outlineMaterial=null
      outlineSize=0
      outlineColor=<rgba 0 0 0 1>
     }
    }
    foregroundColor {colorUseRule=UseColor_Specified, specifiedColor=<rgba 0.5271 0.5271 0.5271 1>}
   }
   slot {class=HorizontalBoxSlot}
  }
  ScrollBox_Results {
   widgetClassPath=&EditorUtilityScrollBox
   bIsVariable=true
   parent=VerticalBox_ScanResults
   widget=ScrollBox_Results
   slot {class=VerticalBoxSlot, properties {verticalAlignment=VAlign_Top, padding=<ltrb 6 10 0 0>}}
  }
  VerticalBox_Edit {
   widgetClassPath=&VerticalBox
   bIsVariable=true
   parent=VerticalBox_0
   widget=VerticalBox_Edit
   slot {class=VerticalBoxSlot, properties {verticalAlignment=VAlign_Top}}
  }
  Border {
   widgetClassPath=&Border
   parent=VerticalBox_Edit
   widget=Border
   properties {
    background {
     bIsDynamicallyLoaded=false
     drawAs=Border
     imageType=NoImage
     mirroring=NoMirror
     resourceName=None
     resourceObject=null
     tiling=NoTile
     imageSize=<32 32>
     margin=<ltrb 0 0 0 1>
     outlineSettings {
      bUseBrushTransparency=false
      roundingType=HalfHeightRadius
      width=0
      color {colorUseRule=UseColor_Specified, specifiedColor=<rgba 0 0 0 0>}
      cornerRadii=<0 0 0 0>
     }
     tintColor {colorUseRule=UseColor_Specified, specifiedColor=<rgba 0.1302 0.1302 0.1302 1>}
     uVRegion {bIsValid=false, max=<0 0>, min=<0 0>}
    }
    padding=<ltrb 6 10 0 0>
   }
   slot {class=VerticalBoxSlot}
  }
  TextBlock_4 {
   widgetClassPath=&TextBlock
   parent=Border
   widget=TextBlock_4
   properties {
    text=Edit :
    font {
     bForceMonospaced=false
     bMaterialIsStencil=false
     fontMaterial=null
     fontObject=null
     letterSpacing=0
     monospacedWidth=1
     size=9
     skewAmount=0
     typefaceFontName=RLSmall-Bold
     outlineSettings {
      bApplyOutlineToDropShadows=false
      bMiteredCorners=false
      bSeparateFillAlpha=false
      outlineMaterial=null
      outlineSize=0
      outlineColor=<rgba 0 0 0 1>
     }
    }
   }
   slot {class=BorderSlot, properties {padding=<ltrb 6 10 0 0>}}
  }
  EditorUtilityExpandableArea_EditCurves {
   widgetClassPath=&EditorUtilityExpandableArea
   bIsVariable=true
   parent=VerticalBox_Edit
   widget=EditorUtilityExpandableArea_EditCurves
   properties {visibility=Collapsed}
   slot {class=VerticalBoxSlot, properties {verticalAlignment=VAlign_Center, padding=<ltrb 0 10 0 0>}}
  }
  TextBlock_5 {
   widgetClassPath=&TextBlock
   namedSlotHost=EditorUtilityExpandableArea_EditCurves
   widget=TextBlock_5
   properties {
    text=Curves
    toolTipText=Is animation used in any ProxyTable Asset
    font {
     bForceMonospaced=false
     bMaterialIsStencil=false
     fontMaterial=null
     fontObject=null
     letterSpacing=0
     monospacedWidth=1
     size=9
     skewAmount=0
     typefaceFontName=RLSmall-Bold
     outlineSettings {
      bApplyOutlineToDropShadows=false
      bMiteredCorners=false
      bSeparateFillAlpha=false
      outlineMaterial=null
      outlineSize=0
      outlineColor=<rgba 0 0 0 1>
     }
    }
   }
  }
  VerticalBox_EditCurves {
   widgetClassPath=&VerticalBox
   bIsVariable=true
   namedSlotHost=EditorUtilityExpandableArea_EditCurves
   widget=VerticalBox_EditCurves
  }
  Spacer_3 {
   widgetClassPath=&Spacer
   parent=VerticalBox_Edit
   widget=Spacer_3
   properties {size=<1 40>}
   slot {class=VerticalBoxSlot, properties {horizontalAlignment=HAlign_Left, verticalAlignment=VAlign_Center}}
  }
  VerticalBox {
   widgetClassPath=&VerticalBox
   parent=VerticalBox_Edit
   widget=VerticalBox
   slot {class=VerticalBoxSlot, properties {horizontalAlignment=HAlign_Left, verticalAlignment=VAlign_Center, padding=<ltrb 80 0 50 0>}}
  }
  Button_Apply {widgetClassPath=&EditorUtilityButton, parent=VerticalBox, widget=Button_Apply, slot {class=VerticalBoxSlot}}
  TextBlock_6 {
   widgetClassPath=&TextBlock
   parent=Button_Apply
   widget=TextBlock_6
   properties {
    text=Apply Changes
    font {
     bForceMonospaced=false
     bMaterialIsStencil=false
     fontMaterial=null
     fontObject=null
     letterSpacing=0
     monospacedWidth=1
     size=10.5
     skewAmount=0
     typefaceFontName=RLSmall-Bold
     outlineSettings {
      bApplyOutlineToDropShadows=false
      bMiteredCorners=false
      bSeparateFillAlpha=false
      outlineMaterial=null
      outlineSize=0
      outlineColor=<rgba 0 0 0 1>
     }
    }
   }
   slot {class=ButtonSlot, properties {horizontalAlignment=HAlign_Fill, verticalAlignment=VAlign_Fill, padding=<ltrb 20 6 20 6>}}
  }
  EditorUtilityScrollBoxPrevis {
   widgetClassPath=&EditorUtilityScrollBox
   parent=VerticalBox_0
   widget=EditorUtilityScrollBoxPrevis
   properties {visibility=Collapsed}
   slot {class=VerticalBoxSlot}
  }
  TextBlock_Sample_item {
   widgetClassPath=&TextBlock
   bIsVariable=true
   parent=EditorUtilityScrollBoxPrevis
   widget=TextBlock_Sample_item
   properties {
    text="Previs(Doesn't render)"
    visibility=Collapsed
    font {
     bForceMonospaced=false
     bMaterialIsStencil=false
     fontMaterial=null
     fontObject=null
     letterSpacing=0
     monospacedWidth=1
     size=9
     skewAmount=0
     typefaceFontName=RLSmall-Regular
     outlineSettings {
      bApplyOutlineToDropShadows=false
      bMiteredCorners=false
      bSeparateFillAlpha=false
      outlineMaterial=null
      outlineSize=0
      outlineColor=<rgba 0 0 0 1>
     }
    }
   }
   slot {class=ScrollBoxSlot}
  }
  HorizontalBox_712 {
   widgetClassPath=&HorizontalBox
   parent=EditorUtilityScrollBoxPrevis
   widget=HorizontalBox_712
   slot {class=ScrollBoxSlot, properties {verticalAlignment=VAlign_Center}}
  }
  Border_456 {
   widgetClassPath=&Border
   parent=HorizontalBox_712
   widget=Border_456
   properties {
    horizontalAlignment=HAlign_Center
    verticalAlignment=VAlign_Center
    background {
     bIsDynamicallyLoaded=false
     drawAs=RoundedBox
     imageType=NoImage
     mirroring=NoMirror
     resourceName=None
     resourceObject=null
     tiling=NoTile
     imageSize=<32 32>
     margin=<ltrb 1 0 0 1>
     outlineSettings {
      bUseBrushTransparency=false
      roundingType=FixedRadius
      width=1
      color {colorUseRule=UseColor_Specified, specifiedColor=<rgba 1 1 1 1>}
      cornerRadii=<0 0 0 0>
     }
     tintColor {colorUseRule=UseColor_Specified, specifiedColor=<rgba 1 1 1 0>}
     uVRegion {bIsValid=false, max=<0 0>, min=<0 0>}
    }
    padding=<ltrb 1 1 1 1>
   }
   slot {class=HorizontalBoxSlot}
  }
  AssetThumbnailWidget_342 {
   widgetClassPath=/Script/UMGEditor.AssetThumbnailWidget
   bIsVariable=true
   parent=Border_456
   widget=AssetThumbnailWidget_342
   properties {
    resolution=<48 48>
    thumbnailSettings {
     bAllowAssetSpecificThumbnailOverlay=false
     bAllowHintText=true
     bAllowRealTimeOnHovered=true
     bForceGenericThumbnail=false
     bOverrideAssetTypeColor=false
     colorStripOrientation=VerticalRightEdge
     genericThumbnailSize=48
     highlightedTextDelegate="(null).None"
     thumbnailLabel=ClassName
     assetTypeColorOverride=<rgba 0 0 0 0>
     hintColorAndOpacity=<rgba 0 0 0 0>
     padding=<ltrb 0 0 0 0>
    }
   }
   slot {class=BorderSlot, properties {horizontalAlignment=HAlign_Center, verticalAlignment=VAlign_Center, padding=<ltrb 1 1 1 1>}}
  }
  TextBlock_Sample_header {
   widgetClassPath=&TextBlock
   bIsVariable=true
   parent=HorizontalBox_712
   widget=TextBlock_Sample_header
   properties {
    text=Anim_SequenceWhatever
    font {
     bForceMonospaced=false
     bMaterialIsStencil=false
     fontMaterial=null
     fontObject=null
     letterSpacing=0
     monospacedWidth=1
     size=10.5
     skewAmount=0
     typefaceFontName=RLSmall-Bold
     outlineSettings {
      bApplyOutlineToDropShadows=false
      bMiteredCorners=false
      bSeparateFillAlpha=false
      outlineMaterial=null
      outlineSize=1
      outlineColor=<rgba 0 0 0 1>
     }
    }
   }
   slot {class=HorizontalBoxSlot, properties {verticalAlignment=VAlign_Center, padding=<ltrb 1 1 1 1>}}
  }
  HorizontalBox_767 {
   widgetClassPath=&HorizontalBox
   parent=HorizontalBox_712
   widget=HorizontalBox_767
   slot {class=HorizontalBoxSlot, properties {verticalAlignment=VAlign_Center, size {sizeRule=Fill, value=1}}}
  }
  EditorUtilityButton_Sample {
   widgetClassPath=&EditorUtilityButton
   parent=HorizontalBox_712
   widget=EditorUtilityButton_Sample
   properties {widgetStyle {
     clickedSlateSound {resourceObject=null}
     disabled {
      bIsDynamicallyLoaded=false
      drawAs=RoundedBox
      imageType=NoImage
      mirroring=NoMirror
      resourceName=None
      resourceObject=null
      tiling=NoTile
      imageSize=<0 0>
      margin=<ltrb 0 0 0 0>
      outlineSettings {
       bUseBrushTransparency=false
       roundingType=FixedRadius
       width=1
       color {colorUseRule=UseColor_ColorTable, specifiedColor=<rgba 0 0 0 0>}
       cornerRadii=<4 4 4 4>
      }
      tintColor {colorUseRule=UseColor_Specified, specifiedColor=<rgba 0.0395 0.0395 0.0395 1>}
      uVRegion {bIsValid=false, max=<0 0>, min=<0 0>}
     }
     disabledForeground {colorUseRule=UseColor_ColorTable, specifiedColor=<rgba 0 0 0 0>}
     hovered {
      bIsDynamicallyLoaded=false
      drawAs=Image
      imageType=NoImage
      mirroring=NoMirror
      resourceName=None
      resourceObject=&/icon_toolbar_genericfinder_40px
      tiling=NoTile
      imageSize=<24 24>
      margin=<ltrb 0 0 0 0>
      outlineSettings {
       bUseBrushTransparency=false
       roundingType=FixedRadius
       width=1
       color {colorUseRule=UseColor_ColorTable, specifiedColor=<rgba 0 0 0 0>}
       cornerRadii=<4 4 4 4>
      }
      tintColor {colorUseRule=UseColor_Specified, specifiedColor=<rgba 1 1 1 1>}
      uVRegion {bIsValid=false, max=<0 0>, min=<0 0>}
     }
     hoveredForeground {colorUseRule=UseColor_ColorTable, specifiedColor=<rgba 0 0 0 0>}
     hoveredSlateSound {resourceObject=null}
     normal {
      bIsDynamicallyLoaded=false
      drawAs=Image
      imageType=NoImage
      mirroring=NoMirror
      resourceName=None
      resourceObject=&icon_toolbar_genericfinder_40px
      tiling=NoTile
      imageSize=<24 24>
      margin=<ltrb 0 0 0 0>
      outlineSettings {
       bUseBrushTransparency=false
       roundingType=FixedRadius
       width=1
       color {colorUseRule=UseColor_ColorTable, specifiedColor=<rgba 0 0 0 0>}
       cornerRadii=<4 4 4 4>
      }
      tintColor {colorUseRule=UseColor_Specified, specifiedColor=<rgba 0.5851 0.5851 0.5851 1>}
      uVRegion {bIsValid=false, max=<0 0>, min=<0 0>}
     }
     normalForeground {colorUseRule=UseColor_ColorTable, specifiedColor=<rgba 0 0 0 0>}
     normalPadding=<ltrb 12 1.5 12 1.5>
     pressed {
      bIsDynamicallyLoaded=false
      drawAs=Image
      imageType=NoImage
      mirroring=NoMirror
      resourceName=None
      resourceObject=&icon_toolbar_genericfinder_40px
      tiling=NoTile
      imageSize=<24 24>
      margin=<ltrb 0 0 0 0>
      outlineSettings {
       bUseBrushTransparency=false
       roundingType=FixedRadius
       width=1
       color {colorUseRule=UseColor_ColorTable, specifiedColor=<rgba 0 0 0 0>}
       cornerRadii=<4 4 4 4>
      }
      tintColor {colorUseRule=UseColor_Specified, specifiedColor=<rgba 1 1 1 1>}
      uVRegion {bIsValid=false, max=<0 0>, min=<0 0>}
     }
     pressedForeground {colorUseRule=UseColor_ColorTable, specifiedColor=<rgba 0 0 0 0>}
     pressedPadding=<ltrb 12 2.5 12 0.5>
     pressedSlateSound {resourceObject=null}
    }}
   slot {class=HorizontalBoxSlot, properties {horizontalAlignment=HAlign_Center, verticalAlignment=VAlign_Center}}
  }
 ]
}
