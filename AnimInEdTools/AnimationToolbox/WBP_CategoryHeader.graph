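# WBP_CategoryHeader is a clickable, collapsible-category header widget in the AnimInEdTools Animation Toolbox, showing a colored line, chevron and label.
# SetupHeader stores the category name, sets the label, tints the button and line with the given color, and picks a right or down chevron from isCollapsed.
# Clicking the button broadcasts OnHeaderClicked with the stored name; the widget itself never toggles collapse state, leaving that to listeners.
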WidgetBlueprint /AnimInEdTools/AnimationToolbox/WBP_CategoryHeader
graphs [
 EventGraph {nodes [
   "On Clicked (Btn_Header)"() {is=title K2Node_ComponentBoundEvent}
   > Call On Header Clicked!(execute:Exec, self:WBP Category Header Object Reference, CategoryName:String=Get categoryName_0() -> (categoryName_0:String)) -> (then:Exec) {is=title K2Node_CallDelegate}
   Event PreConstruct()
   Event Construct()
   Event Tick()
  ]}
 OnHeaderClicked {nodes [OnHeaderClicked!() -> (then, CategoryName:String) {is=title K2Node_FunctionEntry}]}
 SetupHeader {nodes [
   n1 = SetupHeader!() -> (then, LabelText:String, Color:Linear Color Structure, isCollapsed:bool=~) {is=title K2Node_FunctionEntry}
   > Set categoryName_0(categoryName_0:String=@n1) -> (Output_Get:String)
   > "SetText (Text)"(
    self=Get TB_CategoryLabel() -> (TB_CategoryLabel:Text Object Reference)
    InText=To Text!(Value:String=@n1) -> (ReturnValue:Text) {is=title K2Node_GenericToText}
   )
   > SetColorAndOpacity(self:Button Object Reference=Get Btn_Header() -> (Btn_Header:Button Object Reference), InColorAndOpacity:Linear Color Structure=@n1.Color)
   > Set BrushColor!(
    execute
    BrushColor:Linear Color Structure=@n1.Color
    self:Border Object Reference=Get Border_Line() -> (Border_Line:Border Object Reference)
   ) -> (then, Output_Get:Linear Color Structure)
   > n2 = Branch(Condition=@n1.isCollapsed)
   > "SetText (Text)"(self=Get TB_Chevron() -> (TB_Chevron:Text Object Reference), InText="LOCTEXT(\"▶\")")
   "SetText (Text)"(execute=@n2.else, self=Get TB_Chevron(), InText="LOCTEXT(\"▼\")")
   Make Slate Color(SpecifiedColor=@n1.Color)
  ]}
]
variables [
 categoryName_0 {type=string, default=""}
 OnHeaderClicked {}
]
widgets {
 info {inheritedWidgetCount=0, parentClass=&UserWidget, rootWidgetClass=&VerticalBox, widgetCount=6}
 widgets [
  VBox_Root {widgetClassPath=&VerticalBox, widget=VBox_Root}
  Border_Line {
   widgetClassPath=&Border
   bIsVariable=true
   parent=VBox_Root
   widget=Border_Line
   properties {brushColor=<rgba 1 1 1 0.3>, padding=<ltrb 0 1 0 1>}
   slot {class=VerticalBoxSlot, properties {padding=<ltrb 0 8 0 0>}}
  }
  Btn_Header {
   widgetClassPath=&Button
   parent=VBox_Root
   widget=Btn_Header
   properties {
    cursor=Hand
    isFocusable=false
    colorAndOpacity=<rgba 1 0.3112 0.2645 1>
    widgetStyle {
     clickedSlateSound {resourceObject=null}
     disabled {
      bIsDynamicallyLoaded=false
      drawAs=NoDrawType
      imageType=NoImage
      mirroring=NoMirror
      resourceName=None
      resourceObject=null
      tiling=NoTile
      imageSize=<54 54>
      margin=<ltrb 0 0 0 0>
      outlineSettings {
       bUseBrushTransparency=false
       roundingType=FixedRadius
       width=0
       color {colorUseRule=UseColor_Specified, specifiedColor=<rgba 0 0 0 1>}
       cornerRadii=<0 0 0 0>
      }
      tintColor {colorUseRule=UseColor_Specified, specifiedColor=<rgba 1 1 1 0>}
      uVRegion {bIsValid=false, max=<0 0>, min=<0 0>}
     }
     disabledForeground {colorUseRule=UseColor_Specified, specifiedColor=<rgba 1 1 1 1>}
     hovered {
      bIsDynamicallyLoaded=false
      drawAs=NoDrawType
      imageType=NoImage
      mirroring=NoMirror
      resourceName=None
      resourceObject=null
      tiling=NoTile
      imageSize=<54 54>
      margin=<ltrb 0 0 0 0>
      outlineSettings {
       bUseBrushTransparency=false
       roundingType=FixedRadius
       width=0
       color {colorUseRule=UseColor_Specified, specifiedColor=<rgba 0 0 0 1>}
       cornerRadii=<0 0 0 0>
      }
      tintColor {colorUseRule=UseColor_Specified, specifiedColor=<rgba 1 1 1 0.15>}
      uVRegion {bIsValid=false, max=<0 0>, min=<0 0>}
     }
     hoveredForeground {colorUseRule=UseColor_Specified, specifiedColor=<rgba 1 1 1 1>}
     hoveredSlateSound {resourceObject=null}
     normal {
      bIsDynamicallyLoaded=false
      drawAs=NoDrawType
      imageType=NoImage
      mirroring=NoMirror
      resourceName=None
      resourceObject=null
      tiling=NoTile
      imageSize=<54 54>
      margin=<ltrb 0 0 0 0>
      outlineSettings {
       bUseBrushTransparency=false
       roundingType=FixedRadius
       width=0
       color {colorUseRule=UseColor_Specified, specifiedColor=<rgba 0 0 0 1>}
       cornerRadii=<0 0 0 0>
      }
      tintColor {colorUseRule=UseColor_Specified, specifiedColor=<rgba 1 1 1 0>}
      uVRegion {bIsValid=false, max=<0 0>, min=<0 0>}
     }
     normalForeground {colorUseRule=UseColor_Specified, specifiedColor=<rgba 1 1 1 1>}
     normalPadding=<ltrb 0 0 0 0>
     pressed {
      bIsDynamicallyLoaded=false
      drawAs=NoDrawType
      imageType=NoImage
      mirroring=NoMirror
      resourceName=None
      resourceObject=null
      tiling=NoTile
      imageSize=<54 54>
      margin=<ltrb 0 0 0 0>
      outlineSettings {
       bUseBrushTransparency=false
       roundingType=FixedRadius
       width=0
       color {colorUseRule=UseColor_Specified, specifiedColor=<rgba 0 0 0 1>}
       cornerRadii=<0 0 0 0>
      }
      tintColor {colorUseRule=UseColor_Specified, specifiedColor=<rgba 1 1 1 0.08>}
      uVRegion {bIsValid=false, max=<0 0>, min=<0 0>}
     }
     pressedForeground {colorUseRule=UseColor_Specified, specifiedColor=<rgba 1 1 1 1>}
     pressedPadding=<ltrb 0 0 0 0>
     pressedSlateSound {resourceObject=null}
    }
   }
   slot {class=VerticalBoxSlot, properties {padding=<ltrb 0 1 0 1>}}
  }
  HBox_Header {
   widgetClassPath=&HorizontalBox
   parent=Btn_Header
   widget=HBox_Header
   slot {class=ButtonSlot, properties {horizontalAlignment=HAlign_Left}}
  }
  TB_Chevron {
   widgetClassPath=&TextBlock
   bIsVariable=true
   parent=HBox_Header
   widget=TB_Chevron
   properties {
    text="▼"
    font {
     bForceMonospaced=false
     bMaterialIsStencil=false
     fontMaterial=null
     fontObject=&Roboto
     letterSpacing=0
     monospacedWidth=1
     size=12.75
     skewAmount=0
     typefaceFontName=Bold
     outlineSettings {
      bApplyOutlineToDropShadows=false
      bMiteredCorners=false
      bSeparateFillAlpha=false
      outlineMaterial=null
      outlineSize=0
      outlineColor=<rgba 0 0 0 1>
     }
    }
   }
   slot {class=HorizontalBoxSlot, properties {horizontalAlignment=HAlign_Right, verticalAlignment=VAlign_Center, padding=<ltrb 8 0 4 0>}}
  }
  TB_CategoryLabel {
   widgetClassPath=&TextBlock
   bIsVariable=true
   parent=HBox_Header
   widget=TB_CategoryLabel
   properties {
    text=Category
    font {
     bForceMonospaced=false
     bMaterialIsStencil=false
     fontMaterial=null
     fontObject=&Roboto
     letterSpacing=0
     monospacedWidth=1
     size=7.5
     skewAmount=0
     typefaceFontName=Bold
     outlineSettings {
      bApplyOutlineToDropShadows=false
      bMiteredCorners=false
      bSeparateFillAlpha=false
      outlineMaterial=null
      outlineSize=0
      outlineColor=<rgba 0 0 0 1>
     }
    }
    renderTransform {angle=0, scale=<1 1>, shear=<0 0>, translation=<0 3>}
   }
   slot {class=HorizontalBoxSlot, properties {verticalAlignment=VAlign_Center, padding=<ltrb 4 3 4 6>, size {sizeRule=Fill, value=1}}}
  }
 ]
}
