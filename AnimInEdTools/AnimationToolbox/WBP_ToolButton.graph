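# Tool button widget for the AnimInEdTools Animation Toolbox, showing a tool's name, description, icon, optional checkbox and documentation link.
# SetupButton stores the tool's data, falls back to name initials without an icon, and shows the ">" launcher for one EToolAction type.
# Pressing broadcasts row, widget and Python data through OnToolButtonPressed; checkbox changes broadcast OnCheckChanged.
# The launcher arrow animates through text and font swaps; the doc button opens its URL.
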WidgetBlueprint /AnimInEdTools/AnimationToolbox/WBP_ToolButton
graphs [
 EventGraph {nodes [
   "On Hovered (ToolButton)"() {is=title K2Node_ComponentBoundEvent}
   > Switch on EToolAction!(
    execute
    Selection:EToolAction Enum=Byte to Enum EToolAction!(Byte:Byte=Get ToolAction() -> (ToolAction:Byte=0)) -> (ReturnValue:EToolAction Enum) {is=title K2Node_CastByteToEnum}
   ) -> (NewEnumerator0:Exec, NewEnumerator1:Exec) {is=title K2Node_SwitchEnum}
   > Set Text(Text="LOCTEXT(\" >\")", self=Get TB_TabLauncher() -> (TB_TabLauncher:Text Object Reference))
   > Set Font!(
    execute
    Font:Slate Font Info Structure=MakeSlateFontInfo!(
     FontObject:Object Reference=/LauncherSocial/BrutalType
     FontMaterial:Object Reference
     OutlineSettings:Font Outline Settings Structure
     TypefaceFontName:Name=Bold
     Size:float=10
     LetterSpacing:int=0
     SkewAmount:float=0
     bForceMonospaced:bool=false
     bMaterialIsStencil:bool=false
     MonospacedWidth:float=1
    ) -> (ReturnValue:Slate Font Info Structure)
    self:Text Object Reference=Get TB_TabLauncher()
   ) -> (then, Output_Get:Slate Font Info Structure)
   "On Clicked (Btn_DocLink)"() {is=title K2Node_ComponentBoundEvent}
   > LaunchURL!(execute, URL:String=Get DocumentationURL() -> (DocumentationURL:String)) -> (then)
   "On Unhovered (ToolButton)"() {is=title K2Node_ComponentBoundEvent}
   > Switch on EToolAction(Selection=Byte to Enum EToolAction(Byte=Get ToolAction() -> (ToolAction=0)))
   > Set Text(Text="LOCTEXT(\">\")", self=Get TB_TabLauncher())
   > Set Font(
    Font=MakeSlateFontInfo(
     FontObject=&BrutalType
     TypefaceFontName=Light
     Size=10
     LetterSpacing=0
     SkewAmount=0
     bForceMonospaced=false
     bMaterialIsStencil=false
     MonospacedWidth=1
    )
    self=Get TB_TabLauncher()
   )
   "On Pressed (ToolButton)"() {is=title K2Node_ComponentBoundEvent}
   > Call On Tool Button Pressed!(
    execute:Exec
    self:WBP Tool Button Object Reference
    RowName:Name=Get RowName() -> (RowName:Name=None)
    WidgetBP:Editor Utility Widget Blueprint Object Reference=Get WidgetBP() -> (WidgetBP:Editor Utility Widget Blueprint Object Reference)
    PythonPath:String=Get PythonPath() -> (PythonPath:String)
    PythonCommand:String=Get PythonCommand() -> (PythonCommand:String)
   ) -> (then:Exec) {is=title K2Node_CallDelegate}
   > Switch on EToolAction(Selection=Byte to Enum EToolAction(Byte=Get ToolAction() -> (ToolAction=0)))
   n1 = Get TB_TabLauncher()
   > Set Text(Text="LOCTEXT(\"  >\")", self=@n1)
   > Delay(Duration=0.05)
   > Set Text(Text="LOCTEXT(\"   >\")", self=@n1)
   > Delay(Duration=0.05)
   > Set Text(Text="LOCTEXT(\" >\")", self=@n1)
   Event PreConstruct()
   Event Construct()
   Event Tick()
  ]}
 HandleCheckChanged {nodes [
   n1 = HandleCheckChanged!() -> (then, bIsChecked:bool=~) {is=title K2Node_FunctionEntry}
   > Call On Check Changed!(
    execute:Exec
    self:WBP Tool Button Object Reference=Self-Reference()
    bIsChecked:bool=@n1
    ButtonName:String=Get NiceName() -> (NiceName:String)
   ) -> (then:Exec) {is=title K2Node_CallDelegate}
  ]}
 OnCheckChanged {nodes [OnCheckChanged!() -> (then, bIsChecked:bool=~, ButtonName:String) {is=title K2Node_FunctionEntry}]}
 OnToolButtonPressed {nodes [OnToolButtonPressed!() -> (then, RowName:Name, WidgetBP:Editor Utility Widget Blueprint Object Reference, PythonPath:String, PythonCommand:String) {is=title K2Node_FunctionEntry}]}
 SetCheckboxVisibility {nodes [
   n1 = SetCheckboxVisibility!() -> (then, bShow:bool=~) {is=title K2Node_FunctionEntry}
   > n2 = Branch(Condition=@n1)
   > SetVisibility(self=Get CheckBox_ToolList() -> (CheckBox_ToolList:Check Box Object Reference), InVisibility=Visible)
   SetVisibility(execute=@n2.else, self=Get CheckBox_ToolList(), InVisibility=Collapsed)
  ]}
 SetDescriptionVisibility {nodes [
   n1 = SetDescriptionVisibility!() -> (then, bShow:bool=~) {is=title K2Node_FunctionEntry}
   > n2 = Branch(Condition=@n1)
   > SetVisibility(self=Get TB_Description() -> (TB_Description:Text Object Reference), InVisibility=Visible)
   SetVisibility(execute=@n2.else, self=Get TB_Description(), InVisibility=Collapsed)
  ]}
 SetDocLinkVisibility {nodes [
   n1 = SetDocLinkVisibility!() -> (then, bShow:bool=~) {is=title K2Node_FunctionEntry}
   > n2 = Branch(Condition=@n1)
   n3 = Get DocumentationURL()
   > SetVisibility(
    self=Get Btn_DocLink() -> (Btn_DocLink:Button Object Reference)
    InVisibility=Select(Option 0=Visible, Option 1=Collapsed, Index=IsEmpty(InString=@n3))
   )
   SetVisibility(execute=@n2.else, self=Get Btn_DocLink(), InVisibility=Collapsed)
  ]}
 SetIconVisibility {nodes [
   n1 = SetIconVisibility!() -> (then, bShow:bool=~) {is=title K2Node_FunctionEntry}
   > n2 = Branch(Condition=@n1)
   > SetVisibility(self=Get Overlay_Icon() -> (Overlay_Icon:Overlay Object Reference), InVisibility=Visible)
   SetVisibility(execute=@n2.else, self=Get Overlay_Icon(), InVisibility=Collapsed)
  ]}
 SetIsOwned {nodes [
   n1 = SetIsOwned!() -> (then, bOwned:bool=~) {is=title K2Node_FunctionEntry}
   > SetIsChecked!(execute, self:Check Box Object Reference=Get CheckBox_ToolList(), InIsChecked:bool=@n1) -> (then)
  ]}
 SetupButton {nodes [
   n1 = SetupButton!() -> (
    then
    RowName:Name
    TabName:String
    Description:String
    WidgetBP1:Editor Utility Widget Blueprint Object Reference
    Icon:Texture 2D Object Reference
    TintColor:Linear Color Structure
    DocumentationURL:String
    PythonPath:String
    PythonCommand:String
    ToolAction:Byte
    sourceDT:String
   ) {is=title K2Node_FunctionEntry}
   > n2 = Set ToolAction(ToolAction:Byte=@n1.ToolAction) -> (Output_Get:Byte=0)
   > n3 = Switch on EToolAction(Selection=Byte to Enum EToolAction(Byte=@n2))
   n4 = Get TB_TabLauncher()
   > n5 = SetVisibility(self=@n4, InVisibility=Visible)
   Set NiceName(execute=[@n5.then, @n10.then], NiceName:String=@n1.TabName) -> (Output_Get:String)
   > Set RowName(RowName:Name=@n1) -> (Output_Get:Name=None)
   > Set WidgetBP(WidgetBP:Editor Utility Widget Blueprint Object Reference=@n1.WidgetBP1) -> (Output_Get:Editor Utility Widget Blueprint Object Reference)
   > n6 = Is Valid(InputObject=@n1.Icon)
   Set InitialsAccum(execute=@n6.Is Not Valid, InitialsAccum:String) -> (Output_Get:String)
   > n7 = For Each Loop(Array=ParseIntoArray(SourceString=@n1.TabName, Delimiter=" "))
   Set Text(
    execute=@n7.Completed
    Text="To Text (String)"(InString=Get InitialsAccum() -> (InitialsAccum:String))
    self=Get TB_Initials() -> (TB_Initials:Text Object Reference)
   )
   > SetVisibility(self=Get TB_Initials(), InVisibility=Visible)
   > n8 = SetVisibility(self=Get Img_Icon() -> (Img_Icon:Image Object Reference), InVisibility=Collapsed)
   Set Text(execute=[@n9.then, @n8.then], Text="To Text (String)"(InString=@n1.TabName), self=Get TB_TabName() -> (TB_TabName:Text Object Reference))
   > Set Text(Text="To Text (String)"(InString=@n1.Description), self=Get TB_Description())
   > SetToolTipText!(
    execute
    self:Widget Object Reference=Self-Reference()
    InToolTipText:"Text (by ref)"="To Text (String)"(InString=Append!(A=@n1.sourceDT, B="\r\n", C:String=@n1.Description) -> (ReturnValue))
   ) -> (then)
   > Set ColorAndOpacity!(
    execute
    ColorAndOpacity:Linear Color Structure=@n1.TintColor
    self:Button Object Reference=Get ToolButton() -> (ToolButton:Button Object Reference)
   ) -> (then, Output_Get:Linear Color Structure)
   > Set DocumentationURL(DocumentationURL:String=@n1.DocumentationURL) -> (Output_Get:String)
   > SetVisibility(self=Get Btn_DocLink(), InVisibility=Select(Option 0=Visible, Option 1=Collapsed, Index=IsEmpty(InString=@n1.DocumentationURL)))
   > Set PythonPath(PythonPath:String=@n1.PythonPath) -> (Output_Get:String)
   > Set PythonCommand(PythonCommand:String=@n1.PythonCommand) -> (Output_Get:String)
   > Bind Event to On Check State Changed(self:Check Box Object Reference=Get CheckBox_ToolList(), Delegate=Create Event()) {is=title K2Node_AddDelegate}
   Set InitialsAccum(
    execute=@n7.LoopBody
    InitialsAccum=Append!(A=Get InitialsAccum(), B=Left!(SourceString:String=@n7, Count:int=1) -> (ReturnValue:String)) -> (ReturnValue)
   )
   SetBrushFromTexture!(execute=@n6.Is Valid, self:Image Object Reference=Get Img_Icon(), Texture:Texture 2D Object Reference=@n1.Icon, bMatchSize:bool=false) -> (then)
   > SetVisibility(self=Get Img_Icon(), InVisibility=Visible)
   > n9 = SetVisibility(self=Get TB_Initials(), InVisibility=Collapsed)
   n10 = SetVisibility(execute=@n3.NewEnumerator1, self=@n4, InVisibility=Collapsed)
  ]}
]
variables [
 DocumentationURL {type=string, default=""}
 NiceName {type=string, default=""}
 OnCheckChanged {}
 OnToolButtonPressed {}
 PythonCommand {type=string, default=""}
 PythonPath {type=string, default=""}
 RowName {type=string, default=None}
 ToolAction {type=integer}
 WidgetBP {type=EditorUtilityWidgetBlueprint, default=null, class=~, animations=~, widgets=~}
]
widgets {
 info {inheritedWidgetCount=0, parentClass=&UserWidget, rootWidgetClass=&Button, widgetCount=13}
 widgets [
  ToolButton {
   widgetClassPath=&Button
   widget=ToolButton
   properties {widgetStyle {
     clickedSlateSound {resourceObject=null}
     disabled {
      bIsDynamicallyLoaded=false
      drawAs=RoundedBox
      imageType=NoImage
      mirroring=NoMirror
      resourceName=None
      resourceObject=null
      tiling=NoTile
      imageSize=<0 0>
      margin=<ltrb 0 0 0 0>
      outlineSettings {
       bUseBrushTransparency=false
       roundingType=FixedRadius
       width=1
       color {colorUseRule=UseColor_Specified, specifiedColor=<rgba 0.027 0.027 0.031 1>}
       cornerRadii=<5 5 5 5>
      }
      tintColor {colorUseRule=UseColor_Specified, specifiedColor=<rgba 0.015 0.015 0.018 1>}
      uVRegion {bIsValid=false, max=<0 0>, min=<0 0>}
     }
     disabledForeground {colorUseRule=UseColor_Specified, specifiedColor=<rgba 0.4 0.4 0.4 1>}
     hovered {
      bIsDynamicallyLoaded=false
      drawAs=RoundedBox
      imageType=NoImage
      mirroring=NoMirror
      resourceName=None
      resourceObject=null
      tiling=NoTile
      imageSize=<0 0>
      margin=<ltrb 0 0 0 0>
      outlineSettings {
       bUseBrushTransparency=false
       roundingType=FixedRadius
       width=1
       color {colorUseRule=UseColor_Specified, specifiedColor=<rgba 0.027 0.027 0.031 1>}
       cornerRadii=<5 5 5 5>
      }
      tintColor {colorUseRule=UseColor_Specified, specifiedColor=<rgba 0.031 0.031 0.038 1>}
      uVRegion {bIsValid=false, max=<0 0>, min=<0 0>}
     }
     hoveredForeground {colorUseRule=UseColor_Foreground, specifiedColor=<rgba 1 1 1 1>}
     hoveredSlateSound {resourceObject=null}
     normal {
      bIsDynamicallyLoaded=false
      drawAs=RoundedBox
      imageType=NoImage
      mirroring=NoMirror
      resourceName=None
      resourceObject=null
      tiling=NoTile
      imageSize=<0 0>
      margin=<ltrb 0 0 0 0>
      outlineSettings {
       bUseBrushTransparency=false
       roundingType=FixedRadius
       width=1
       color {colorUseRule=UseColor_Specified, specifiedColor=<rgba 0.027 0.027 0.031 1>}
       cornerRadii=<5 5 5 5>
      }
      tintColor {colorUseRule=UseColor_Specified, specifiedColor=<rgba 0.0222 0.0222 0.0252 1>}
      uVRegion {bIsValid=false, max=<0 0>, min=<0 0>}
     }
     normalForeground {colorUseRule=UseColor_Foreground, specifiedColor=<rgba 1 1 1 1>}
     normalPadding=<ltrb 0 0 0 0>
     pressed {
      bIsDynamicallyLoaded=false
      drawAs=RoundedBox
      imageType=NoImage
      mirroring=NoMirror
      resourceName=None
      resourceObject=null
      tiling=NoTile
      imageSize=<0 0>
      margin=<ltrb 0 0 0 0>
      outlineSettings {
       bUseBrushTransparency=false
       roundingType=FixedRadius
       width=1
       color {colorUseRule=UseColor_Specified, specifiedColor=<rgba 0.027 0.027 0.031 1>}
       cornerRadii=<5 5 5 5>
      }
      tintColor {colorUseRule=UseColor_Specified, specifiedColor=<rgba 0.04 0.04 0.048 1>}
      uVRegion {bIsValid=false, max=<0 0>, min=<0 0>}
     }
     pressedForeground {colorUseRule=UseColor_Foreground, specifiedColor=<rgba 1 1 1 1>}
     pressedPadding=<ltrb 0 0 0 0>
     pressedSlateSound {resourceObject=null}
    }}
  }
  HBox_Content {
   widgetClassPath=&HorizontalBox
   parent=ToolButton
   widget=HBox_Content
   slot {class=ButtonSlot, properties {horizontalAlignment=HAlign_Fill, padding=<ltrb 10 1.5 10 1.5>}}
  }
  CheckBox_ToolList {
   widgetClassPath=/Script/UMG.CheckBox
   bIsVariable=true
   parent=HBox_Content
   widget=CheckBox_ToolList
   properties {visibility=Collapsed}
   slot {class=HorizontalBoxSlot, properties {padding=<ltrb 3 0 6 0>}}
  }
  Overlay_Icon {
   widgetClassPath=&Overlay
   bIsVariable=true
   parent=HBox_Content
   widget=Overlay_Icon
   slot {class=HorizontalBoxSlot, properties {horizontalAlignment=HAlign_Left, verticalAlignment=VAlign_Center, padding=<ltrb 0 0 10 0>}}
  }
  Img_Icon {
   widgetClassPath=&Image
   bIsVariable=true
   parent=Overlay_Icon
   widget=Img_Icon
   properties {brush {
     bIsDynamicallyLoaded=false
     drawAs=Image
     imageType=FullColor
     mirroring=NoMirror
     resourceName=None
     resourceObject=null
     tiling=NoTile
     imageSize=<24 24>
     margin=<ltrb 0 0 0 0>
     outlineSettings {
      bUseBrushTransparency=false
      roundingType=FixedRadius
      width=0
      color {colorUseRule=UseColor_Specified, specifiedColor=<rgba 0 0 0 1>}
      cornerRadii=<0 0 0 0>
     }
     tintColor {colorUseRule=UseColor_Specified, specifiedColor=<rgba 1 1 1 1>}
     uVRegion {bIsValid=false, max=<0 0>, min=<0 0>}
    }}
   slot {class=OverlaySlot, properties {horizontalAlignment=HAlign_Center, verticalAlignment=VAlign_Center}}
  }
  TB_Initials {
   widgetClassPath=&TextBlock
   bIsVariable=true
   parent=Overlay_Icon
   widget=TB_Initials
   properties {
    justification=Center
    text=" "
    visibility=Collapsed
    colorAndOpacity {colorUseRule=UseColor_Specified, specifiedColor=<rgba 0.784 0.784 0.831 1>}
    font {
     bForceMonospaced=false
     bMaterialIsStencil=false
     fontMaterial=null
     fontObject=&Roboto
     letterSpacing=0
     monospacedWidth=0
     size=9
     skewAmount=0
     typefaceFontName=Bold
     outlineSettings {
      bApplyOutlineToDropShadows=false
      bMiteredCorners=false
      bSeparateFillAlpha=false
      outlineMaterial=null
      outlineSize=0
      outlineColor=<rgba 0 0 0 1>
     }
    }
   }
   slot {class=OverlaySlot, properties {horizontalAlignment=HAlign_Center, verticalAlignment=VAlign_Center}}
  }
  VBox_Text {
   widgetClassPath=&VerticalBox
   parent=HBox_Content
   widget=VBox_Text
   slot {class=HorizontalBoxSlot, properties {verticalAlignment=VAlign_Center, size {sizeRule=Fill, value=1}}}
  }
  HorizontalBox_0 {
   widgetClassPath=&HorizontalBox
   parent=VBox_Text
   widget=HorizontalBox_0
   slot {class=VerticalBoxSlot, properties {verticalAlignment=VAlign_Top, padding=<ltrb 0 0 0 3>}}
  }
  TB_TabName {
   widgetClassPath=&TextBlock
   bIsVariable=true
   parent=HorizontalBox_0
   widget=TB_TabName
   properties {
    text=Tool Name
    textOverflowPolicy=Ellipsis
    colorAndOpacity {colorUseRule=UseColor_Specified, specifiedColor=<rgba 0.784 0.784 0.831 1>}
    font {
     bForceMonospaced=false
     bMaterialIsStencil=false
     fontMaterial=null
     fontObject=&Roboto
     letterSpacing=0
     monospacedWidth=0
     size=10
     skewAmount=0
     typefaceFontName=Bold
     outlineSettings {
      bApplyOutlineToDropShadows=false
      bMiteredCorners=false
      bSeparateFillAlpha=false
      outlineMaterial=null
      outlineSize=0
      outlineColor=<rgba 0 0 0 1>
     }
    }
   }
   slot {class=HorizontalBoxSlot}
  }
  TB_TabLauncher {
   widgetClassPath=&TextBlock
   bIsVariable=true
   parent=HorizontalBox_0
   widget=TB_TabLauncher
   properties {
    text=">"
    textOverflowPolicy=Ellipsis
    colorAndOpacity {colorUseRule=UseColor_Specified, specifiedColor=<rgba 0.784 0.784 0.831 1>}
    font {
     bForceMonospaced=false
     bMaterialIsStencil=false
     fontMaterial=null
     fontObject=&BrutalType
     letterSpacing=0
     monospacedWidth=0
     size=10
     skewAmount=0
     typefaceFontName=Light
     outlineSettings {
      bApplyOutlineToDropShadows=false
      bMiteredCorners=false
      bSeparateFillAlpha=false
      outlineMaterial=null
      outlineSize=0
      outlineColor=<rgba 0 0 0 1>
     }
    }
    renderTransform {angle=0, scale=<1 1.5>, shear=<0 0>, translation=<0 -2>}
   }
   slot {class=HorizontalBoxSlot, properties {horizontalAlignment=HAlign_Left, padding=<ltrb 6 0 0 0>}}
  }
  TB_Description {
   widgetClassPath=&TextBlock
   bIsVariable=true
   parent=VBox_Text
   widget=TB_Description
   properties {
    text=Description
    textOverflowPolicy=Ellipsis
    colorAndOpacity {colorUseRule=UseColor_Specified, specifiedColor=<rgba 0.416 0.416 0.478 1>}
    font {
     bForceMonospaced=false
     bMaterialIsStencil=false
     fontMaterial=null
     fontObject=&Roboto
     letterSpacing=0
     monospacedWidth=0
     size=8
     skewAmount=0
     typefaceFontName=Regular
     outlineSettings {
      bApplyOutlineToDropShadows=false
      bMiteredCorners=false
      bSeparateFillAlpha=false
      outlineMaterial=null
      outlineSize=0
      outlineColor=<rgba 0 0 0 1>
     }
    }
   }
   slot {class=VerticalBoxSlot, properties {verticalAlignment=VAlign_Top}}
  }
  Btn_DocLink {
   widgetClassPath=&Button
   parent=HBox_Content
   widget=Btn_DocLink
   properties {
    cursor=Hand
    toolTipText="click here to jump to the documentation page!"
    visibility=Collapsed
    colorAndOpacity=<rgba 1 1 1 0.6>
    widgetStyle {
     clickedSlateSound {resourceObject=null}
     disabled {
      bIsDynamicallyLoaded=false
      drawAs=NoDrawType
      imageType=NoImage
      mirroring=NoMirror
      resourceName=None
      resourceObject=null
      tiling=NoTile
      imageSize=<0 0>
      margin=<ltrb 0 0 0 0>
      outlineSettings {
       bUseBrushTransparency=false
       roundingType=FixedRadius
       width=0
       color {colorUseRule=UseColor_Specified, specifiedColor=<rgba 0 0 0 0>}
       cornerRadii=<0 0 0 0>
      }
      tintColor {colorUseRule=UseColor_Specified, specifiedColor=<rgba 0 0 0 0>}
      uVRegion {bIsValid=false, max=<0 0>, min=<0 0>}
     }
     disabledForeground {colorUseRule=UseColor_Specified, specifiedColor=<rgba 0.5271 0.5271 0.5271 1>}
     hovered {
      bIsDynamicallyLoaded=false
      drawAs=NoDrawType
      imageType=NoImage
      mirroring=NoMirror
      resourceName=None
      resourceObject=null
      tiling=NoTile
      imageSize=<0 0>
      margin=<ltrb 0 0 0 0>
      outlineSettings {
       bUseBrushTransparency=false
       roundingType=FixedRadius
       width=0
       color {colorUseRule=UseColor_Specified, specifiedColor=<rgba 0 0 0 0>}
       cornerRadii=<0 0 0 0>
      }
      tintColor {colorUseRule=UseColor_Specified, specifiedColor=<rgba 0 0 0 0>}
      uVRegion {bIsValid=false, max=<0 0>, min=<0 0>}
     }
     hoveredForeground {colorUseRule=UseColor_Specified, specifiedColor=<rgba 0 1 1 1>}
     hoveredSlateSound {resourceObject=null}
     normal {
      bIsDynamicallyLoaded=false
      drawAs=NoDrawType
      imageType=NoImage
      mirroring=NoMirror
      resourceName=None
      resourceObject=null
      tiling=NoTile
      imageSize=<0 0>
      margin=<ltrb 0 0 0 0>
      outlineSettings {
       bUseBrushTransparency=false
       roundingType=FixedRadius
       width=0
       color {colorUseRule=UseColor_Specified, specifiedColor=<rgba 0 0 0 0>}
       cornerRadii=<0 0 0 0>
      }
      tintColor {colorUseRule=UseColor_Specified, specifiedColor=<rgba 0 0 0 0>}
      uVRegion {bIsValid=false, max=<0 0>, min=<0 0>}
     }
     normalForeground {colorUseRule=UseColor_Specified, specifiedColor=<rgba 0 1 1 1>}
     normalPadding=<ltrb 0 0 0 0>
     pressed {
      bIsDynamicallyLoaded=false
      drawAs=NoDrawType
      imageType=NoImage
      mirroring=NoMirror
      resourceName=None
      resourceObject=null
      tiling=NoTile
      imageSize=<0 0>
      margin=<ltrb 0 0 0 0>
      outlineSettings {
       bUseBrushTransparency=false
       roundingType=FixedRadius
       width=0
       color {colorUseRule=UseColor_Specified, specifiedColor=<rgba 0 0 0 0>}
       cornerRadii=<0 0 0 0>
      }
      tintColor {colorUseRule=UseColor_Specified, specifiedColor=<rgba 0 0 0 0>}
      uVRegion {bIsValid=false, max=<0 0>, min=<0 0>}
     }
     pressedForeground {colorUseRule=UseColor_Specified, specifiedColor=<rgba 0 1 1 1>}
     pressedPadding=<ltrb 0 0 0 0>
     pressedSlateSound {resourceObject=null}
    }
   }
   slot {class=HorizontalBoxSlot, properties {verticalAlignment=VAlign_Center, padding=<ltrb 2 0 2 0>}}
  }
  Img_DocLink {
   widgetClassPath=&Image
   bIsVariable=true
   parent=Btn_DocLink
   widget=Img_DocLink
   properties {
    renderOpacity=0.5613
    brush {
     bIsDynamicallyLoaded=false
     drawAs=Image
     imageType=FullColor
     mirroring=NoMirror
     resourceName=None
     resourceObject=/AnimInEdTools/AnimationToolbox/Icons/AnimToolbox_UI_info
     tiling=NoTile
     imageSize=<20 20>
     margin=<ltrb 0 0 0 0>
     outlineSettings {
      bUseBrushTransparency=false
      roundingType=FixedRadius
      width=0
      color {colorUseRule=UseColor_Specified, specifiedColor=<rgba 0 0 0 0>}
      cornerRadii=<0 0 0 0>
     }
     tintColor {colorUseRule=UseColor_Specified, specifiedColor=<rgba 1 1 1 1>}
     uVRegion {bIsValid=false, max=<0 0>, min=<0 0>}
    }
    colorAndOpacity=<rgba 1 1 1 0.8>
   }
   slot {class=ButtonSlot}
  }
 ]
}
