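# Editor utility widget for an Animation Toolbox panel, listing animation tools under colored, collapsible categories that open widget tabs or run Python scripts.
# It builds buttons from toolbox, project and per-user DataTables, or from every table sharing the row struct in show-all mode.
# Inline Python persists display settings, hidden tools and collapsed categories to a JSON config.
# Checking an external tool copies its row into the user's DataTable, creating that table if missing.
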EditorUtilityWidgetBlueprint /AnimInEdTools/AnimationToolbox/Anim_Toolbox
graphs [
 AddButtonToCategory {nodes [
   n1 = AddButtonToCategory!() -> (then, index:int, categoryName:String) {is=title K2Node_FunctionEntry}
   > n2 = Sequence()
   n3 = Do Once(execute=@n12.then_0, Reset=@n2.then_0)
   n4 = Create Widget!(
    execute:Exec=@n6.then
    WorldContextObject:Object Reference
    Class:User Widget Class Reference=&/WBP_CategoryHeader.WBP_CategoryHeader_C
    OwningPlayer:Player Controller Object Reference
   ) -> (then:Exec, ReturnValue:WBP Category Header Object Reference)
   AddChildToVerticalBox(execute=@n3.Completed, self=Get ButtonListBox() -> (ButtonListBox:Vertical Box Object Reference), Content=@n4)
   Set in_CategoryIndex(execute=@n2.then_1, in_CategoryIndex:int=@n1) -> (Output_Get:int=0)
   > Set in_CategoryName(in_CategoryName:String=@n1.categoryName) -> (Output_Get:String)
   > n5 = For Each Loop!(
    Exec
    Array:Array of ST Anim Toolbox Row Composite Structures=Get CombinedDataTableRows() -> (CombinedDataTableRows:Array of ST Anim Toolbox Row Composite Structures)
   ) -> (
    LoopBody
    Array Element_RowName_33_#1:Name
    Array Element_Row_32_#2_NiceName_21_#1:String
    Array Element_Row_32_#2_ToolAction_38_#3:EToolAction Enum
    Array Element_Row_32_#2_WidgetBP_20_#2:Editor Utility Widget Blueprint Object Reference
    Array Element_Row_32_#2_PythonPath_32_#4:String
    Array Element_Row_32_#2_PythonCommand_34_#5:String
    Array Element_Row_32_#2_Category_26_#6:ENUM_ToolboxCategory Enum
    Array Element_Row_32_#2_Icon_16_#7:Texture 2D Object Reference
    Array Element_Row_32_#2_Description_19_#8:String
    Array Element_Row_32_#2_DocumentationURL_29_#9:String
    Array Index
    Completed
   )
   > Branch(Condition="Equal (Integer)"(
     A=Get in_CategoryIndex() -> (in_CategoryIndex:int=0)
     B="To Integer (Byte)"!(InByte:Byte=@n5.Array Element_Row_32_#2_Category_26_#6) -> (ReturnValue:int=0)
    ))
   > Set loc_rowName(loc_rowName:Name=@n5) -> (Output_Get:Name=None)
   > Set loc_NiceName(loc_NiceName:String=@n5.Array Element_Row_32_#2_NiceName_21_#1) -> (Output_Get:String)
   > Set loc_WidgetBP(loc_WidgetBP:Editor Utility Widget Blueprint Object Reference=@n5.Array Element_Row_32_#2_WidgetBP_20_#2) -> (Output_Get:Editor Utility Widget Blueprint Object Reference)
   > Set loc_RowIcon(loc_RowIcon:Texture 2D Object Reference=@n5.Array Element_Row_32_#2_Icon_16_#7) -> (Output_Get:Texture 2D Object Reference)
   > Set loc_RowDescript(loc_RowDescript:String=@n5.Array Element_Row_32_#2_Description_19_#8) -> (Output_Get:String)
   > Set loc_DocURL(loc_DocURL:String=@n5.Array Element_Row_32_#2_DocumentationURL_29_#9) -> (Output_Get:String)
   > Set LocPythonPath(LocPythonPath:String=@n5.Array Element_Row_32_#2_PythonPath_32_#4) -> (Output_Get:String)
   > Set LocPythonCommand(LocPythonCommand:String=@n5.Array Element_Row_32_#2_PythonCommand_34_#5) -> (Output_Get:String)
   > Set loc_ForLoop_Idx(loc_ForLoop_Idx:int=@n5.Array Index) -> (Output_Get:int=0)
   > n6 = Set loc_ToolAction(loc_ToolAction:EToolAction Enum=@n5.Array Element_Row_32_#2_ToolAction_38_#3) -> (Output_Get:EToolAction Enum=NewEnumerator0)
   n7 = Create Widget(execute=@n14.else, Class=&/WBP_ToolButton.WBP_ToolButton_C) -> (ReturnValue:WBP Tool Button Object Reference)
   n8 = AddChildToVerticalBox(execute=@n10.then, self=Get ButtonListBox(), Content=@n7)
   > SetPadding!(execute, self:Vertical Box Slot Object Reference=@n8, InPadding:Margin Structure=(Left=0,Top=4,Right=0,Bottom=4)) -> (then)
   n9 = Self-Reference()
   n10 = Bind Event to On Check Changed(execute=@n11.then, self:WBP Tool Button Object Reference=@n7, Delegate=Create Event(self=@n9)) {is=title K2Node_AddDelegate}
   n11 = Bind Event to On Tool Button Pressed(execute=[@n22.then, @n21.then], self:WBP Tool Button Object Reference=@n7, Delegate=Create Event(self=@n9)) {is=title K2Node_AddDelegate}
   Bind Event to On Header Clicked(execute=@n25.then, self:WBP Category Header Object Reference=@n4, Delegate=Create Event(self=@n9)) {is=title K2Node_AddDelegate}
   > n12 = Sequence()
   n13 = Contains Item(
    TargetArray:Array of Strings=Get collapsedCategories() -> (collapsedCategories:Array of Strings)
    ItemToFind:"String (by ref)"=Get in_CategoryName() -> (in_CategoryName:String)
   )
   n14 = Branch(execute=@n12.then_1, Condition=@n13)
   n15 = "Get (a copy)"!(
    Array:Array of ST Anim Toolbox Row Composite Structures=Get CombinedDataTableRows()
    Dimension 1=Get loc_ForLoop_Idx() -> (loc_ForLoop_Idx:int=0)
   ) -> (Output_RowName_33_#1:Name, Output_Row_32_#2:ST Anim Toolbox Widget Tab Structure)
   n16 = Get CombinedDataTableSources_User() -> (CombinedDataTableSources_User:Array of ST Anim Toolbox Source Tag Structures)
   n17 = Find Item!(
    TargetArray:Array of ST Anim Toolbox Source Tag Structures=@n16
    ItemToFind_CategoryTag_6_#10:EToolSource Enum=NewEnumerator2
    ItemToFind_Data_5_#11:ST Anim Toolbox Widget Tab Structure=@n15.Output_Row_32_#2
   ) -> (ReturnValue)
   n18 = Find Item(
    TargetArray=@n16
    ItemToFind_CategoryTag_6_#10=NewEnumerator1
    ItemToFind_Data_5_#11=@n15.Output_Row_32_#2
   )
   n19 = Find Item(
    TargetArray=@n16
    ItemToFind_CategoryTag_6_#10=NewEnumerator0
    ItemToFind_Data_5_#11=@n15.Output_Row_32_#2
   )
   SetupButton!(
    execute=@n7.then
    self:WBP Tool Button Object Reference=@n7
    RowName:Name=Get loc_rowName() -> (loc_rowName:Name=None)
    TabName:String=Get loc_NiceName() -> (loc_NiceName:String)
    Description:String=Get loc_RowDescript() -> (loc_RowDescript:String)
    WidgetBP1:Editor Utility Widget Blueprint Object Reference=Get loc_WidgetBP() -> (loc_WidgetBP:Editor Utility Widget Blueprint Object Reference)
    Icon:Texture 2D Object Reference=Get loc_RowIcon() -> (loc_RowIcon:Texture 2D Object Reference)
    TintColor:Linear Color Structure="Get (a copy)"!(
     Array:Array of Linear Color Structures=Get CategoryColors() -> (CategoryColors:Array of Linear Color Structures)
     Dimension 1=Get in_CategoryIndex() -> (in_CategoryIndex=0)
    ) -> (Output:Linear Color Structure)
    DocumentationURL:String=Get loc_DocURL() -> (loc_DocURL:String)
    PythonPath:String=Get LocPythonPath() -> (LocPythonPath:String)
    PythonCommand:String=Get LocPythonCommand() -> (LocPythonCommand:String)
    ToolAction:Byte=Get loc_ToolAction() -> (loc_ToolAction:EToolAction Enum=NewEnumerator0)
    sourceDT:String="Get (a ref)"(
     Array:Array of Strings=Get CombinedDataTableSources_List() -> (CombinedDataTableSources_List:Array of Strings)
     Dimension 1=Get loc_ForLoop_Idx() -> (loc_ForLoop_Idx=0)
    ) -> (Output:"String (by ref)")
   ) -> (then)
   > n20 = Branch(Condition=OR Boolean!(A="integer >= integer"(A=@n17), B="integer >= integer"(A=@n19), C:bool="integer >= integer"(A=@n18)) -> (ReturnValue))
   > n21 = SetIsOwned!(execute, self:WBP Tool Button Object Reference=@n7, bOwned:bool=true) -> (then)
   n22 = SetIsOwned(execute=@n20.else, self=@n7, bOwned=false)
   n23 = Get in_CategoryIndex() -> (in_CategoryIndex=0)
   n24 = Get CategoryColors()
   n25 = SetupHeader!(
    execute=@n4.then
    self:WBP Category Header Object Reference=@n4
    LabelText:String="Get (a copy)"(Array:Array of Strings=Get CategoryNames() -> (CategoryNames:Array of Strings), Dimension 1=@n23) -> (Output:String)
    Color:Linear Color Structure=LinearColor * LinearColor(
     A:Linear Color Structure="Get (a copy)"(Array:Array of Linear Color Structures=@n24, Dimension 1=@n23) -> (Output:Linear Color Structure)
     B:Linear Color Structure=(R=0.5443,G=0.5443,B=0.5443,A=1)
    ) -> (ReturnValue:Linear Color Structure) {is=title K2Node_PromotableOperator}
    isCollapsed:bool=@n13
   ) -> (then)
  ]}
 CombineDataTables_All {nodes [
   CombineDataTables_All() {is=title K2Node_FunctionEntry}
   > n1 = Sequence()
   > Clear(TargetArray:Array of ST Anim Toolbox Row Composite Structures=Get CombinedDataTableRows())
   > Clear(TargetArray:Array of Strings=Get CombinedDataTableSources_List())
   > Set DataTable_Path(DataTable_Path:String) -> (Output_Get:String)
   n2 = Get EditorAssetSubsystem() -> (ReturnValue:Editor Asset Subsystem Object Reference) {is=title K2Node_GetEditorSubsystem}
   n3 = ListAssetsByTagValue!(
    execute=@n1.then_1
    self:Editor Asset Subsystem Object Reference=@n2
    TagName:Name=RowStructure
    TagValue:String=&/ST_Anim_Toolbox_Widget_Tab
   ) -> (then, ReturnValue:Array of Strings)
   > n4 = Set ALL_DataTablePaths(ALL_DataTablePaths:Array of Strings=@n3) -> (Output_Get:Array of Strings)
   > n5 = For Each Loop!(Exec, Array:Array of Strings=@n4) -> (LoopBody, Array Element:String, Array Index, Completed)
   > n6 = LoadAsset!(execute, self:Editor Asset Subsystem Object Reference=@n2, AssetPath=@n5) -> (then, ReturnValue)
   > n7 = Cast To DataTable!(execute, Object=@n6) -> (then, CastFailed, AsData Table:Data Table Object Reference) {is=title K2Node_DynamicCast}
   > n8 = GetDataTableRowNames(Table=@n7)
   > n9 = For Each Loop(Array:Array of Names=@n8) -> (Array Element:Name)
   > n10 = Get Data Table Row!(execute:Exec, DataTable:Data Table Object Reference=@n7, RowName:Name=@n9) -> (then:Exec, RowNotFound:Exec, ReturnValue:ST Anim Toolbox Widget Tab Structure)
   > Add Unique!(
    execute
    TargetArray:Array of ST Anim Toolbox Row Composite Structures=Get CombinedDataTableRows()
    NewItem_RowName_33_#1:Name=@n9
    NewItem_Row_32_#2:ST Anim Toolbox Widget Tab Structure=@n10
   ) -> (then, ReturnValue)
   > n11 = get_DT_Name!(execute, self, Object:Object Reference=@n7) -> (then, Name:String)
   > Add(TargetArray=Get CombinedDataTableSources_List(), NewItem=@n11) {is=title K2Node_CallArrayFunction}
   Set ALL_UserDataTable(ALL_UserDataTable:Data Table Object Reference) -> (Output_Get:Data Table Object Reference)
  ]}
 CombineDataTables_Default {nodes [
   CombineDataTables_Default() {is=title K2Node_FunctionEntry}
   > Clear(TargetArray:Array of ST Anim Toolbox Source Tag Structures=Get CombinedDataTableSources_User())
   > Clear(TargetArray:Array of Strings=Get CombinedDataTableSources_List())
   > n1 = GetDataTableRowNames(Table=Get ToolboxDataTable() -> (ToolboxDataTable:Data Table Object Reference))
   > n2 = For Each Loop(Array=@n1)
   n3 = GetDataTableRowNames(execute=@n2.Completed, Table=Get ProjectDataTable() -> (ProjectDataTable:Data Table Object Reference))
   > n4 = For Each Loop(Array=@n3)
   n5 = GetDataTableRowNames(execute=@n4.Completed, Table=Get UserDataTable() -> (UserDataTable:Data Table Object Reference))
   > n6 = For Each Loop(Array=@n5)
   ExecutePythonCommand(
    execute=[@n17.Completed, @n6.Completed]
    PythonCommand="import unreal; eus=unreal.get_editor_subsystem(unreal.EditorUtilitySubsystem); bp=unreal.load_asset('&/EUW_Anim_Toolbox'); w=eus.find_utility_widget_from_blueprint(bp); hidden=set(w.get_editor_property('hiddenToolRows')) if w else set(); rows=list(w.get_editor_property('combinedDataTableRows')) if w else []; srcs=list(w.get_editor_property('combinedDataTableSources_User')) if w else []; indices=[i for i,r in enumerate(rows) if r.get_editor_property('row').get_editor_property('nice Name') in hidden]; [rows.pop(i) for i in sorted(indices,reverse=True)]; [srcs.pop(i) for i in sorted([i for i in indices if i < len(srcs)],reverse=True)]; w.set_editor_property('combinedDataTableRows',rows) if w else None; w.set_editor_property('combinedDataTableSources_User',srcs) if w else None; print('Filter removed '+str(len(indices))+' rows, hidden='+str(hidden))"
   )
   n7 = Get Data Table Row(execute=@n6.LoopBody, DataTable=Get UserDataTable(), RowName=@n6)
   > Set CDTD_TableRow(CDTD_TableRow:ST Anim Toolbox Widget Tab Structure=@n7) -> (Output_Get:ST Anim Toolbox Widget Tab Structure)
   > Set CDTD_CombinedDT_Len(CDTD_CombinedDT_Len:int=Length(TargetArray:Array of ST Anim Toolbox Row Composite Structures=Get CombinedDataTableRows())) -> (Output_Get:int=0)
   > Add Unique(
    TargetArray=Get CombinedDataTableRows()
    NewItem_RowName_33_#1=@n6
    NewItem_Row_32_#2=Get CDTD_TableRow() -> (CDTD_TableRow:ST Anim Toolbox Widget Tab Structure)
   )
   n8 = Get CombinedDataTableRows()
   > Branch(Condition="Not Equal (Integer)"(A=Length(TargetArray=@n8), B=Get CDTD_CombinedDT_Len() -> (CDTD_CombinedDT_Len:int=0)))
   > Add!(
    execute
    TargetArray:Array of ST Anim Toolbox Source Tag Structures=Get CombinedDataTableSources_User()
    NewItem_CategoryTag_6_#10:EToolSource Enum=NewEnumerator1
    NewItem_Data_5_#11:ST Anim Toolbox Widget Tab Structure=Get CDTD_TableRow()
   ) -> (then, ReturnValue)
   > n9 = get_DT_Name(Object=Get UserDataTable())
   > Add!(execute, TargetArray:Array of Strings=Get CombinedDataTableSources_List(), NewItem:"String (by ref)"=@n9) -> (then, ReturnValue)
   n10 = Get Data Table Row(execute=@n4.LoopBody, DataTable=Get ProjectDataTable(), RowName=@n4)
   > Set CDTD_TableRow(CDTD_TableRow=@n10)
   > Set CDTD_CombinedDT_Len(CDTD_CombinedDT_Len=Length(TargetArray=Get CombinedDataTableRows())) -> (Output_Get=0)
   > Add Unique(
    TargetArray=Get CombinedDataTableRows()
    NewItem_RowName_33_#1=@n4
    NewItem_Row_32_#2=Get CDTD_TableRow()
   )
   n11 = Get CombinedDataTableRows()
   > Branch(Condition="Not Equal (Integer)"(A=Length(TargetArray=@n11), B=Get CDTD_CombinedDT_Len() -> (CDTD_CombinedDT_Len=0)))
   > Add!(
    execute
    TargetArray:Array of ST Anim Toolbox Source Tag Structures=Get CombinedDataTableSources_User()
    NewItem_CategoryTag_6_#10:EToolSource Enum=NewEnumerator0
    NewItem_Data_5_#11:ST Anim Toolbox Widget Tab Structure=Get CDTD_TableRow()
   ) -> (then, ReturnValue)
   > n12 = get_DT_Name(Object=Get ProjectDataTable())
   > Add!(execute, TargetArray:Array of Strings=Get CombinedDataTableSources_List(), NewItem:"String (by ref)"=@n12) -> (then, ReturnValue)
   n13 = Get Data Table Row(execute=@n2.LoopBody, DataTable=Get ToolboxDataTable(), RowName=@n2)
   > Set CDTD_TableRow(CDTD_TableRow=@n13)
   n14 = Get CombinedDataTableRows()
   > Set CDTD_CombinedDT_Len(CDTD_CombinedDT_Len=Length(TargetArray=@n14)) -> (Output_Get=0)
   > Add Unique(
    TargetArray=@n14
    NewItem_RowName_33_#1=@n2
    NewItem_Row_32_#2=Get CDTD_TableRow()
   )
   n15 = Get CombinedDataTableRows()
   > Branch(Condition="Not Equal (Integer)"(A=Length(TargetArray=@n15), B=Get CDTD_CombinedDT_Len() -> (CDTD_CombinedDT_Len=0)))
   > Add!(
    execute
    TargetArray:Array of ST Anim Toolbox Source Tag Structures=Get CombinedDataTableSources_User()
    NewItem_CategoryTag_6_#10:EToolSource Enum=NewEnumerator2
    NewItem_Data_5_#11:ST Anim Toolbox Widget Tab Structure=Get CDTD_TableRow()
   ) -> (then, ReturnValue)
   > n16 = get_DT_Name(Object=Get ToolboxDataTable())
   > Add!(execute, TargetArray:Array of Strings=Get CombinedDataTableSources_List(), NewItem:"String (by ref)"=@n16) -> (then, ReturnValue)
   n17 = For Each Loop!(Exec, Array:Array of ST Anim Toolbox Source Tag Structures=Get CombinedDataTableSources_User()) -> (
    LoopBody
    Array Element_CategoryTag_6_#10:EToolSource Enum
    Array Element_Data_5_#11_NiceName_21_#1:String
    Array Element_Data_5_#11_ToolAction_38_#3:EToolAction Enum
    Array Element_Data_5_#11_WidgetBP_20_#2:Editor Utility Widget Blueprint Object Reference
    Array Element_Data_5_#11_PythonPath_32_#4:String
    Array Element_Data_5_#11_PythonCommand_34_#5:String
    Array Element_Data_5_#11_Category_26_#6:ENUM_ToolboxCategory Enum
    Array Element_Data_5_#11_Icon_16_#7:Texture 2D Object Reference
    Array Element_Data_5_#11_Description_19_#8:String
    Array Element_Data_5_#11_DocumentationURL_29_#9:String
    Array Index
    Completed
   )
   > PrintString(InString=Append!(
     A=Enum to String!(Enumerator:EToolSource Enum=@n17) -> (ReturnValue:String)
     B=": "
     C:String=@n17.Array Element_Data_5_#11_NiceName_21_#1
    ) -> (ReturnValue))
   PrintString(InString=Append!(A="User Tool: ", B="To String (Name)"(InName=@n6), C=" from DT: ", D:String=@n9) -> (ReturnValue))
   PrintString(InString=Append(A="Toolbox Tool: ", B="To String (Name)"(InName=@n2), C=" from DT: ", D=@n16))
   PrintString(InString=Append(A="Project Tool: ", B="To String (Name)"(InName=@n4), C=" from DT: ", D=@n12))
  ]}
 EventGraph {nodes [
   "On Clicked (SettingsButton)"() {is=title K2Node_ComponentBoundEvent}
   n1 = Get Name!(self:EUW Context Menu Object Reference=Get EUW_ContextMenu() -> (EUW_ContextMenu:EUW Context Menu Object Reference)) -> (Name:String)
   n2 = Get EUW_ContextMenu()
   > n3 = Branch(Condition=AND Boolean(A=IsVisible(self=@n2), B="Equal Exactly (String)"(A=@n1, B=Settings)))
   > Show!(execute, self:EUW Context Menu Object Reference=Get EUW_ContextMenu(), Show:bool=false, Mouse Position:Vector 2D Structure, Name:String) -> (then)
   Clear Context Menu(execute=@n3.else, self:EUW Context Menu Object Reference=Get EUW_ContextMenu())
   > n4 = Sequence!(execute) -> (then_0, then_1, then_2:Exec, then_3:Exec, then_4:Exec, then_5:Exec, then_6:Exec)
   > LoadSettings()
   n5 = Add Item!(
    execute=@n4.then_1
    self:EUW Context Menu Object Reference=Get EUW_ContextMenu()
    Event:"Delegate (by ref)"=Create Event()
    Slot Widget:User Widget Class Reference=/AnimInEdTools/ContextMenu/EUW_ContextMenu_ItemB.EUW_ContextMenu_ItemB_C
    Category:String=Vis
   ) -> (then, Context Menu Item:EUW Context Menu Item Object Reference, Widget:Editor Utility Widget Object Reference)
   > n6 = Cast To EUW_ContextMenu_ItemB!(execute, Object=@n5.Widget) -> (then, CastFailed, AsEUW Context Menu Item B:EUW Context Menu Item B Object Reference) {is=title K2Node_DynamicCast}
   > Set String Data!(execute, String Data:String=Icons, self:EUW Context Menu Item B Object Reference=@n6) -> (then, Output_Get:String)
   > SetIsChecked!(execute, self:EUW Context Menu Item B Object Reference=@n6, bIsChecked:bool=Get Setting_ShowIcons() -> (Setting_ShowIcons:bool)) -> (then)
   n7 = Add Item(
    execute=@n4.then_2
    self=Get EUW_ContextMenu()
    Event=Create Event()
    Slot Widget=&EUW_ContextMenu_ItemB_C
    Category=Vis
   )
   > n8 = Cast To EUW_ContextMenu_ItemB(Object=@n7.Widget)
   > Set String Data(String Data=Documentation Links, self=@n8)
   > SetIsChecked(self=@n8, bIsChecked=Get Setting_ShowDocLinks() -> (Setting_ShowDocLinks:bool))
   n9 = Add Item(
    execute=@n4.then_3
    self=Get EUW_ContextMenu()
    Event=Create Event()
    Slot Widget=&EUW_ContextMenu_ItemB_C
    Category=Vis
   )
   > n10 = Cast To EUW_ContextMenu_ItemB(Object=@n9.Widget)
   > Set String Data(String Data=Descriptions, self=@n10)
   > SetIsChecked(self=@n10, bIsChecked=Get Setting_ShowDescriptions() -> (Setting_ShowDescriptions:bool))
   n11 = Add Item(
    execute=@n4.then_4
    self=Get EUW_ContextMenu()
    Event=Create Event()
    Slot Widget=&EUW_ContextMenu_ItemB_C
    Category=Config
   )
   > n12 = Cast To EUW_ContextMenu_ItemB(Object=@n11.Widget)
   > n13 = Branch(Condition=Get Setting_ToolList() -> (Setting_ToolList:bool))
   > n14 = Set String Data(String Data=Show My Tools, self=@n12)
   SetIsChecked(execute=[@n15.then, @n14.then], self=@n12, bIsChecked=false)
   > SetVisibility(
    self=Get CheckBox_IsChecked!(self:EUW Context Menu Item B Object Reference=@n12) -> (CheckBox_IsChecked:Check Box Object Reference)
    InVisibility=Collapsed
   )
   n15 = Set String Data(execute=@n13.else, String Data=Show All Tools, self=@n12)
   n16 = Add Item(
    execute=@n4.then_5
    self=Get EUW_ContextMenu()
    Event=Create Event()
    Slot Widget=&EUW_ContextMenu_ItemB_C
    Category=Learn More
   )
   > n17 = Cast To EUW_ContextMenu_ItemB(Object=@n16.Widget)
   > Set String Data(String Data=AnimToolbox Confluence, self=@n17)
   > SetIsChecked(self=@n17, bIsChecked=false)
   > SetVisibility(self=Get CheckBox_IsChecked(self=@n17), InVisibility=Collapsed)
   n18 = GetCachedGeometry!(self:Widget Object Reference=Get SettingsButton() -> (SettingsButton:Button Object Reference)) -> (ReturnValue:"Geometry Structure (by ref)")
   n19 = BreakVector2D(InVec=GetLocalSize!(Geometry:"Geometry Structure (by ref)"=@n18) -> (ReturnValue:Vector 2D Structure))
   Show(
    execute=@n4.then_6
    self=Get EUW_ContextMenu()
    Show=true
    Mouse Position=vector2d + vector2d(
     A:Vector 2D Structure="Local to Absolute (Coordinates)"!(Geometry:"Geometry Structure (by ref)"=@n18, LocalCoordinate:Vector 2D Structure) -> (ReturnValue:Vector 2D Structure)
     B:Vector 2D Structure=MakeVector2D(X=0, Y=@n19.Y)
    ) -> (ReturnValue:Vector 2D Structure) {is=title K2Node_PromotableOperator}
    Name=Settings
   )
   OnDescriptionClicked!() -> (OutputDelegate, then, Context Item:Editor Utility Widget Object Reference, Slot Widget:Editor Utility Widget Object Reference) {is=title K2Node_CustomEvent}
   > Set Setting_ShowDescriptions(Setting_ShowDescriptions:bool=NOT Boolean(A=Get Setting_ShowDescriptions())) -> (Output_Get:bool)
   > SaveSettings()
   > LoadSettings()
   OnToolListClicked!() -> (OutputDelegate, then, Context Item:Editor Utility Widget Object Reference, Slot Widget:Editor Utility Widget Object Reference) {is=title K2Node_CustomEvent}
   > n20 = Branch(Condition=Get Setting_ToolList())
   Set Setting_ToolList(execute=[@n20.else, @n20.then], Setting_ToolList:bool=NOT Boolean(A=Get Setting_ToolList())) -> (Output_Get:bool)
   > RebuildButtonList()
   > ExecutePythonCommand(PythonCommand="import unreal; eus=unreal.get_editor_subsystem(unreal.EditorUtilitySubsystem); bp=unreal.load_asset('&EUW_Anim_Toolbox'); w=eus.find_utility_widget_from_blueprint(bp); in_show_all=not bool(w.get_editor_property('setting_ToolList')) if w else False; box=w.get_editor_property('buttonListBox') if (w and in_show_all) else None; hidden=set(w.get_editor_property('hiddenToolRows')) if (w and in_show_all) else set(); [c.get_editor_property('checkBox_ToolList').set_checked_state(unreal.CheckBoxState.UNCHECKED, True) for c in (box.get_all_children() if box else []) if c.get_class().get_name()=='WBP_ToolButton_C' and c.get_editor_property('niceName') in hidden]; print('Restored unchecked for: '+str(hidden))")
   OnConfluenceClicked!() -> (OutputDelegate, then, Context Item:Editor Utility Widget Object Reference, Slot Widget:Editor Utility Widget Object Reference) {is=title K2Node_CustomEvent}
   > Execute Python Script(PythonScript="import webbrowser\r\n\r\nwebbrowser.open(\"https://epicgames.atlassian.net/wiki/spaces/FORT/pages/310775107/Animation+Toolbox\")")
   Debug_ReadHiddenTools() {is=title K2Node_CustomEvent}
   > ExecutePythonCommand(PythonCommand="import unreal; eus=unreal.get_editor_subsystem(unreal.EditorUtilitySubsystem); bp=unreal.load_asset('&EUW_Anim_Toolbox'); w=eus.find_utility_widget_from_blueprint(bp); box=w.get_editor_property('buttonListBox') if w else None; lines=[]; [lines.append('cls={} name={} checked={}'.format(c.get_class().get_name(), c.get_editor_property('niceName'), c.get_editor_property('checkBox_ToolList').get_checked_state())) for c in (box.get_all_children() if box else []) if c.get_class().get_name()=='WBP_ToolButton_C']; report='\\n'.join(lines); print(report); w.set_editor_property('debug_HiddenTools', report) if w else None")
   > n21 = Set Debug_HiddenTools(Debug_HiddenTools:String=Get Debug_HiddenTools() -> (Debug_HiddenTools:String)) -> (Output_Get:String)
   > PrintString(InString=@n21)
   OnDocLinksToggled_0!() -> (OutputDelegate, then, Context Item:Editor Utility Widget Object Reference, Slot Widget:Editor Utility Widget Object Reference) {is=title K2Node_CustomEvent}
   > Set Setting_ShowDocLinks(Setting_ShowDocLinks:bool=NOT Boolean(A=Get Setting_ShowDocLinks())) -> (Output_Get:bool)
   > SaveSettings()
   > LoadSettings()
   Event Construct()
   > n22 = GetProjectDataTable!(execute, self) -> (then, DataTable:Data Table Object Reference)
   > Set ProjectDataTable(ProjectDataTable:Data Table Object Reference=@n22) -> (Output_Get:Data Table Object Reference)
   > n23 = GetUserDataTable!(execute, self) -> (then, DataTable:Data Table Object Reference)
   > Set UserDataTable(UserDataTable:Data Table Object Reference=@n23) -> (Output_Get:Data Table Object Reference)
   > ExecutePythonCommand(PythonCommand="import unreal\neas = unreal.get_editor_subsystem(unreal.EditorAssetSubsystem)\neus = unreal.get_editor_subsystem(unreal.EditorUtilitySubsystem)\nbp  = unreal.load_asset('&EUW_Anim_Toolbox')\nw   = eus.find_utility_widget_from_blueprint(bp)\ndev_dir = unreal.Paths.game_user_developer_dir() if w else ''\nparts = [p for p in dev_dir.replace('\\\\\\\\', '/').split('/') if p]\nusername = parts[-1] if parts else 'Unknown'\nstruct_tag = '&ST_Anim_Toolbox_Widget_Tab'\ncandidates = eas.list_assets_by_tag_value('RowStructure', struct_tag) if w else []\nexisting = [c for c in candidates if f'/Developers/{username}/AnimTools' in c]\ndt = unreal.load_asset(existing[0].split('.')[0]) if existing else None\nif dt:\n  print(f'[EnsureUserToolboxDT] Found: {existing[0]}')\n  if w: w.set_editor_property('userDataTable', dt)\nelse:\n  print(f'[EnsureUserToolboxDT] No user DT found for {username} \\u2014 will create on first external tool bookmark')")
   > Clear(TargetArray:Array of ST Anim Toolbox Row Composite Structures=Get CombinedDataTableRows())
   > LoadSettings()
   > RebuildButtonList()
   > ExecutePythonCommand(PythonCommand="import unreal; eus=unreal.get_editor_subsystem(unreal.EditorUtilitySubsystem); bp=unreal.load_asset('&EUW_Anim_Toolbox'); w=eus.find_utility_widget_from_blueprint(bp); ctx=w.get_editor_property('eUW_ContextMenu') if w else None; print('InitContextMenu ok, ctx=', ctx)")
   > ExecutePythonCommand(PythonCommand="import unreal; eus=unreal.get_editor_subsystem(unreal.EditorUtilitySubsystem); bp=unreal.load_asset('&EUW_Anim_Toolbox'); w=eus.find_utility_widget_from_blueprint(bp); in_show_all=not bool(w.get_editor_property('setting_ToolList')) if w else False; box=w.get_editor_property('buttonListBox') if (w and in_show_all) else None; hidden=set(w.get_editor_property('hiddenToolRows')) if (w and in_show_all) else set(); [c.get_editor_property('checkBox_ToolList').set_checked_state(unreal.CheckBoxState.UNCHECKED, True) for c in (box.get_all_children() if box else []) if c.get_class().get_name()=='WBP_ToolButton_C' and c.get_editor_property('niceName') in hidden]; print('Restored unchecked for: '+str(hidden))")
   Event Tick()
  ]}
 Get Projects {nodes [
   Get Projects() {is=title K2Node_FunctionEntry}
   > Return Node!(execute, Projects:Array of BP Animation Browser Project Object References)
  ]}
 GetProjectDataTable {nodes [
   n1 = LoadGameFromSlot!(execute=@n15.then, SlotName:String=AnimationBrowser, UserIndex:int=0) -> (then, ReturnValue:Save Game Object Reference)
   n2 = PrintString(InString=GetObjectName(Object=@n1))
   n3 = Cast To SG_AnimationBrowser!(execute=[@n1.then, @n2.then], Object=@n1) -> (then, CastFailed, AsSG Animation Browser:SG Animation Browser Object Reference) {is=title K2Node_DynamicCast}
   n4 = LoadAsset_Blocking(
    execute=[@n13.then, @n3.then]
    Asset=Get Project!(self:SG Animation Browser Object Reference=@n3) -> (Project:BP Animation Browser Project Soft Object Reference)
   )
   n5 = Cast To BP_AnimationBrowser_Project!(execute=[@n4.then, @n14.then], Object=@n4) -> (then, CastFailed, AsBP Animation Browser Project:BP Animation Browser Project Object Reference) {is=title K2Node_DynamicCast}
   > Set AB_Project(AB_Project:BP Animation Browser Project Object Reference=@n5) -> (Output_Get:BP Animation Browser Project Object Reference)
   n6 = Get EditorAssetSubsystem()
   > n7 = ListAssetsByTagValue(self=@n6, TagName=RowStructure, TagValue=&ST_Anim_Toolbox_Widget_Tab)
   > n8 = For Each Loop!(Exec, Array:Array of Strings=@n7) -> (LoopBody, Array Element:String, Array Index, Completed)
   n9 = LoadAsset(execute=@n8.Completed, self=@n6, AssetPath=Get DataTablePath() -> (DataTablePath:String))
   > n10 = Cast To DataTable(Object=@n9)
   > Return Node!(execute, DataTable:Data Table Object Reference=@n10)
   n11 = Get AB_Project() -> (AB_Project:BP Animation Browser Project Object Reference)
   n12 = "Get (a copy)"(Array:Array of Strings=ParseIntoArray(SourceString=Get Object Path String(Object=@n11), Delimiter=/)) -> (Output:String)
   Branch(execute=@n8.LoopBody, Condition=Contains(SearchIn=@n8, Substring=Append!(A=/, B=@n12, C=/AnimTools) -> (ReturnValue)))
   > Set DataTablePath(DataTablePath:String=@n8) -> (Output_Get:String)
   n13 = PrintString(InString=CastSG: ok)
   n14 = PrintString(InString=GetObjectName(Object=@n4))
   n15 = GetProjectDataTable() {is=title K2Node_FunctionEntry}
  ]}
 GetUserDataTable {nodes [
   n1 = GameUserDeveloperDir!() -> (ReturnValue:String)
   PrintString(InString=Append!(A="DevDir: ", B=@n1) -> (ReturnValue))
   n2 = ParseIntoArray(SourceString=@n1, Delimiter=/)
   n3 = "Get (a copy)"(Array=@n2, Dimension 1=Last Index!(TargetArray=@n2) -> (ReturnValue=0))
   > PrintString(InString=Append(A="Username: ", B=@n3))
   n4 = Append(A=Append(A=/, B=@n3), B=/AnimTools)
   > n5 = PrintString(InString=Append(A="Filter: ", B=@n4))
   n6 = Get EditorAssetSubsystem()
   n7 = ListAssetsByTagValue(execute=[@n5.then, @n13.then], self=@n6, TagName=RowStructure, TagValue=&ST_Anim_Toolbox_Widget_Tab)
   > n8 = For Each Loop(Array=@n7)
   n9 = LoadAsset(execute=[@n12.then, @n8.Completed], self=@n6, AssetPath=Get UserDataTablePath() -> (UserDataTablePath:String))
   > n10 = Cast To DataTable(Object=@n9)
   > Return Node(DataTable=@n10)
   Branch(execute=[@n11.then, @n8.LoopBody], Condition=Contains(SearchIn=@n8, Substring=@n4))
   > Set UserDataTablePath(UserDataTablePath:String=@n8) -> (Output_Get:String)
   n11 = PrintString(InString=Append(A="Candidate: ", B=@n8))
   n12 = PrintString(InString=Append(A="Result: ", B=Get UserDataTablePath()))
   GetUserDataTable() {is=title K2Node_FunctionEntry}
   > n13 = Set UserDataTablePath()
  ]}
 Load ContextMenu Additional Project Options {nodes [
   Load ContextMenu Additional Project Options!() -> (then, Context:Name) {is=title K2Node_FunctionEntry}
   > Return Node!(execute, HadOptions:bool=false)
  ]}
 LoadSettings {nodes [
   LoadSettings() {is=title K2Node_FunctionEntry}
   > ExecutePythonCommand(PythonCommand="import json, os, unreal; p=os.path.join(os.path.expanduser('~'),'Documents','UE_Prefs','AnimTools_config.json'); cfg=json.load(open(p)) if os.path.exists(p) else {}; eus=unreal.get_editor_subsystem(unreal.EditorUtilitySubsystem); bp=unreal.load_asset('&EUW_Anim_Toolbox'); w=eus.find_utility_widget_from_blueprint(bp); w.set_editor_property('setting_ShowIcons',cfg.get('ShowIcons',True)) if w else None; w.set_editor_property('setting_ShowDocLinks',cfg.get('ShowDocLinks',True)) if w else None; w.set_editor_property('setting_ShowDescriptions',cfg.get('ShowDescriptions',True)) if w else None; hidden=cfg.get('HiddenTools',[]); w.set_editor_property('hiddenToolRows',hidden) if w else None; collapsed=cfg.get('CollapsedCategories',[]); w.set_editor_property('collapsedCategories',collapsed) if w else None; print('LoadSettings hidden='+str(hidden)+' collapsed='+str(collapsed))")
   > RefreshUI(self=Self-Reference())
  ]}
 OnButtonCheckChanged {nodes [
   n1 = OnButtonCheckChanged!() -> (then, bIsChecked:bool=~, ButtonName:String) {is=title K2Node_FunctionEntry}
   > Set Temp_CheckChangedName(Temp_CheckChangedName:String=@n1.ButtonName) -> (Output_Get:String)
   > Set Temp_CheckChangedState(Temp_CheckChangedState:bool=@n1) -> (Output_Get:bool)
   > ExecutePythonCommand(PythonCommand="import json, os, unreal\neas = unreal.get_editor_subsystem(unreal.EditorAssetSubsystem)\neus = unreal.get_editor_subsystem(unreal.EditorUtilitySubsystem)\nbp  = unreal.load_asset('&EUW_Anim_Toolbox')\nw   = eus.find_utility_widget_from_blueprint(bp)\nbtn = str(w.get_editor_property('temp_CheckChangedName')) if w else ''\nchk = bool(w.get_editor_property('temp_CheckChangedState')) if w else True\nhidden = list(w.get_editor_property('hiddenToolRows')) if w else []\nhidden = [h for h in hidden if h != btn]\nhidden = hidden if chk else (hidden + [btn])\np = os.path.join(os.path.expanduser('~'), 'Documents', 'UE_Prefs', 'AnimTools_config.json')\nos.makedirs(os.path.dirname(p), exist_ok=True)\ncfg = json.load(open(p)) if os.path.exists(p) else {}\ncfg['HiddenTools'] = hidden\njson.dump(cfg, open(p, 'w'), indent=2)\nif w: w.set_editor_property('hiddenToolRows', hidden)\nprint('CheckChanged: btn=' + btn + ' chk=' + str(chk))\nif chk and w:\n  user_dt    = w.get_editor_property('userDataTable')\n  proj_dt    = w.get_editor_property('projectDataTable')\n  toolbox_dt = w.get_editor_property('toolboxDataTable')\n  def pkg(dt): return dt.get_path_name().split('.')[0] if dt else ''\n  known_pkgs = {pkg(proj_dt), pkg(toolbox_dt), pkg(user_dt)} - {''}\n  struct_tag = '&ST_Anim_Toolbox_Widget_Tab'\n  all_pkgs = eas.list_assets_by_tag_value('RowStructure', struct_tag)\n  external_pkgs = [p2.split('.')[0] for p2 in all_pkgs if p2.split('.')[0] not in known_pkgs]\n  if not external_pkgs:\n    print('[CheckChanged] No external DTs')\n  else:\n    ## find matching row in an external DT by NiceName\n    matched_row = None\n    matched_key = None\n    for ext_pkg in external_pkgs:\n      ext_dt = unreal.load_asset(ext_pkg)\n      if not ext_dt: continue\n      ext_json = json.loads(unreal.DataTableFunctionLibrary.export_data_table_to_json_string(ext_dt) or '[]')\n      for row in ext_json:\n        if row.get('Nice Name', '') == btn:\n          matched_row = row\n          matched_key = row.get('Name', '')\n          break\n      if matched_row: break\n    if not matched_row:\n      print('[CheckChanged] No external row matched nicename=' + btn + ' (owned or not external)')\n    else:\n      ## lazy-create user DT if it doesn't exist yet\n      if not user_dt:\n        dev_dir = unreal.Paths.game_user_developer_dir()\n        parts = [p3 for p3 in dev_dir.replace('\\\\\\\\', '/').split('/') if p3]\n        username = parts[-1] if parts else 'Unknown'\n        user_dt_dir = f'/Game/Developers/{username}/AnimTools/'\n        dt_name = f'DT_Toolbox_Tabs_{username}'\n        dt_pkg_path = f'{user_dt_dir}{dt_name}'\n        struct_asset = unreal.load_asset('&ST_Anim_Toolbox_Widget_Tab')\n        eas.make_directory(user_dt_dir)\n        factory = unreal.DataTableFactory()\n        factory.set_editor_property('struct', struct_asset)\n        af = unreal.AssetToolsHelpers.get_asset_tools()\n        user_dt = af.create_asset(dt_name, user_dt_dir, unreal.DataTable, factory)\n        if user_dt:\n          eas.save_asset(dt_pkg_path)\n          w.set_editor_property('userDataTable', user_dt)\n          print('[CheckChanged] Created user DT: ' + dt_pkg_path)\n          unreal.EditorDialog.show_message(\n            'Anim Toolbox - Personal DataTable Created',\n            f'A personal tool list DataTable was created for you at:\\n\\n{dt_pkg_path}\\n\\nPlease check this file into source control so your tool choices are saved and synced.',\n            unreal.AppMsgType.OK\n          )\n        else:\n          print('[CheckChanged] ERROR: failed to create user DT at ' + dt_pkg_path)\n      ## write the row into user DT (existing_keys check accounts for newly created empty DT)\n      if user_dt:\n        existing_keys = set(str(n) for n in user_dt.get_row_names())\n        if matched_key in existing_keys:\n          print('[CheckChanged] Already in user DT: ' + matched_key)\n        else:\n          user_json = json.loads(unreal.DataTableFunctionLibrary.export_data_table_to_json_string(user_dt) or '[]')\n          user_json.append(matched_row)\n          ok = unreal.DataTableFunctionLibrary.fill_data_table_from_json_string(user_dt, json.dumps(user_json))\n          if ok:\n            eas.save_asset(pkg(user_dt))\n            print('[CheckChanged] Promoted: nicename=' + btn + ' rowkey=' + matched_key)\n          else:\n            print('[CheckChanged] fill failed for: ' + matched_key)")
  ]}
 OnCategoryHeaderClicked_0 {nodes [
   n1 = OnCategoryHeaderClicked_0!() -> (then, CategoryName:String) {is=title K2Node_FunctionEntry}
   > Set clickedCategoryName(clickedCategoryName:String=@n1) -> (Output_Get:String)
   > ExecutePythonCommand(PythonCommand="import json, os, unreal; p=os.path.join(os.path.expanduser('~'),'Documents','UE_Prefs','AnimTools_config.json'); cfg=json.load(open(p)) if os.path.exists(p) else {}; collapsed=cfg.get('CollapsedCategories',[]); eus=unreal.get_editor_subsystem(unreal.EditorUtilitySubsystem); bp=unreal.load_asset('&EUW_Anim_Toolbox'); w=eus.find_utility_widget_from_blueprint(bp); cat=w.get_editor_property('clickedCategoryName') if w else ''; collapsed.remove(cat) if cat in collapsed else collapsed.append(cat); cfg['CollapsedCategories']=collapsed; json.dump(cfg,open(p,'w'),indent=2); w.set_editor_property('collapsedCategories',collapsed) if w else None")
   > RebuildButtonList(self=Self-Reference())
  ]}
 OnIconsToggled_0 {nodes [
   OnIconsToggled_0!() -> (then, Context Item:Editor Utility Widget Object Reference, Slot Widget:Editor Utility Widget Object Reference) {is=title K2Node_FunctionEntry}
   > Set Setting_ShowIcons(Setting_ShowIcons:bool=NOT Boolean(A=Get Setting_ShowIcons())) -> (Output_Get:bool)
   > SaveSettings()
   > LoadSettings() {is=title K2Node_CallFunction}
  ]}
 OnToolButtonClicked {nodes [
   n1 = OnToolButtonClicked!() -> (then, RowName:Name, WidgetBP1:Editor Utility Widget Blueprint Object Reference, PythonPath:String, PythonCommand:String) {is=title K2Node_FunctionEntry}
   > Set OTBC_WidgetBP(OTBC_WidgetBP:Editor Utility Widget Blueprint Object Reference=@n1.WidgetBP1) -> (Output_Get:Editor Utility Widget Blueprint Object Reference)
   > Set OTBC_PythonPath(OTBC_PythonPath:String=@n1.PythonPath) -> (Output_Get:String)
   > Set OTBC_PythonCommand(OTBC_PythonCommand:String=@n1.PythonCommand) -> (Output_Get:String)
   n2 = Get EditorUtilitySubsystem()
   > n3 = FindUtilityWidgetFromBlueprint!(
    execute
    self:Editor Utility Subsystem Object Reference=@n2
    InBlueprint:Editor Utility Widget Blueprint Object Reference=Get OTBC_WidgetBP() -> (OTBC_WidgetBP:Editor Utility Widget Blueprint Object Reference)
   ) -> (then, ReturnValue:Editor Utility Widget Object Reference)
   > n4 = Is Valid(InputObject=@n3)
   n5 = Is Valid(exec=@n4.Is Not Valid, InputObject=Get OTBC_WidgetBP())
   n6 = Branch(execute=@n5.Is Not Valid, Condition=IsEmpty(InString=Get OTBC_PythonCommand() -> (OTBC_PythonCommand:String)))
   n7 = Get OTBC_PythonPath() -> (OTBC_PythonPath:String)
   > Execute Python Script(PythonScript=Append(
     A=Append!(A="import unreal, os; ", B="_f = os.path.join(unreal.Paths.project_dir(), '", C:String=@n7, D:String="'); ") -> (ReturnValue)
     B="exec(compile(open(_f).read(), _f, 'exec'), {'__file__': _f})"
    ))
   n8 = Get OTBC_PythonPath()
   Execute Python Script(
    execute=@n6.else
    PythonScript=Append!(
     A=Append!(A="import unreal, os; ", B="_f = os.path.join(unreal.Paths.project_dir(), '", C=@n8, D:String="'); ") -> (ReturnValue)
     B="exec(open(_f).read()); "
     C=Get OTBC_PythonCommand()
    ) -> (ReturnValue)
   )
   SpawnAndRegisterTab(execute=@n5.Is Valid, self=@n2, InBlueprint=Get OTBC_WidgetBP())
   ReleaseInstanceOfAsset!(execute, self:Editor Utility Subsystem Object Reference=@n2, Asset:Object Reference=Get OTBC_WidgetBP()) -> (then)
  ]}
 RebuildButtonList {nodes [
   RebuildButtonList() {is=title K2Node_FunctionEntry}
   > Clear(TargetArray=Get CombinedDataTableRows())
   > n1 = Branch(Condition=Get Setting_ToolList())
   > n2 = CombineDataTables_All() {is=title K2Node_CallFunction}
   ClearChildren(execute=[@n2.then, @n4.then], self=Get ButtonListBox())
   > n3 = For Each Loop(Array=Get CategoryNames())
   LoadSettings(execute=@n3.Completed)
   AddButtonToCategory!(execute=@n3.LoopBody, self, index:int=@n3.Array Index, categoryName:String=@n3) -> (then) {is=title K2Node_CallFunction}
   n4 = CombineDataTables_Default(execute=@n1.else) {is=title K2Node_CallFunction}
  ]}
 RefreshUI {nodes [
   RefreshUI() {is=title K2Node_FunctionEntry}
   > n1 = For Each Loop(Array:Array of Widget Object References=GetAllChildren(self=Get ButtonListBox())) -> (Array Element:Widget Object Reference)
   > n2 = Cast To WBP_ToolButton!(execute, Object=@n1) -> (then, CastFailed, AsWBP Tool Button:WBP Tool Button Object Reference) {is=title K2Node_DynamicCast}
   > SetIconVisibility!(execute, self:WBP Tool Button Object Reference=@n2, bShow:bool=Get Setting_ShowIcons()) -> (then)
   > SetDocLinkVisibility!(execute, self:WBP Tool Button Object Reference=@n2, bShow:bool=Get Setting_ShowDocLinks()) -> (then)
   > SetDescriptionVisibility!(execute, self:WBP Tool Button Object Reference=@n2, bShow:bool=Get Setting_ShowDescriptions()) -> (then)
   > SetCheckboxVisibility!(execute, self:WBP Tool Button Object Reference=@n2, bShow:bool=Get Setting_ToolList()) -> (then)
  ]}
 SaveSettings {nodes [
   SaveSettings() {is=title K2Node_FunctionEntry}
   > ExecutePythonCommand(PythonCommand="import json, os, unreal; eus=unreal.get_editor_subsystem(unreal.EditorUtilitySubsystem); bp=unreal.load_asset('&EUW_Anim_Toolbox'); w=eus.find_utility_widget_from_blueprint(bp); p=os.path.join(os.path.expanduser('~'),'Documents','UE_Prefs','AnimTools_config.json'); os.makedirs(os.path.dirname(p),exist_ok=True); cfg=json.load(open(p)) if os.path.exists(p) else {}; cfg['ShowIcons']=bool(w.get_editor_property('setting_ShowIcons')) if w else cfg.get('ShowIcons',True); cfg['ShowDocLinks']=bool(w.get_editor_property('setting_ShowDocLinks')) if w else cfg.get('ShowDocLinks',True); cfg['ShowDescriptions']=bool(w.get_editor_property('setting_ShowDescriptions')) if w else cfg.get('ShowDescriptions',True); json.dump(cfg,open(p,'w'),indent=2); print('SaveSettings -> '+p)")
   MakeLiteralString(Value=True)
   MakeLiteralString(Value=True)
   MakeLiteralString(Value=False)
   MakeLiteralString(Value=False)
   MakeLiteralString(Value=True)
   MakeLiteralString(Value=False)
   Get Setting_ShowIcons()
   Get Setting_ShowDocLinks()
   Get Setting_ShowDescriptions()
  ]}
 get_DT_Name {nodes [
   n1 = get_DT_Name!() -> (then, Object:Object Reference) {is=title K2Node_FunctionEntry}
   n2 = GetSystemPath!(Object:Object Reference=@n1) -> (ReturnValue:String)
   > Return Node!(execute, Name:String=Split(SourceString=Split(SourceString=@n2, InStr=/, SearchDir=FromEnd).RightS, InStr=., SearchDir=FromStart))
  ]}
]
variables [
 AB_Project {type=BP_AnimationBrowser_Project_C, default=null}
 CategoryColors {
  type="array<LinearColor>"
  default [
   <rgba 0.1195 0.521 0.4072 1>
   <rgba 0.227 0.5972 0.1195 1>
   <rgba 1 0.3916 0.1119 1>
   <rgba 0.7157 0.0648 0.3916 1>
   <rgba 0.7305 0.0762 0.0762 1>
   <rgba 0.8 0.1948 0.7277 1>
   <rgba 0.2582 0.491 0.8963 1>
   <rgba 0.0699 0.2593 0.7005 1>
   <rgba 0.8714 0.855 0.1559 1>
  ]
 }
 CategoryNames {
  type="array<string>"
  default [Rig Setup, General, Animation, Constraints, "Aligning && Over-Driving", "Cameras && Viewports", Selection, Workflows, User]
 }
 clickedCategoryName {type=string, default=""}
 collapsedCategories {type="array<string>", default []}
 CombinedDataTable {type=DataTable, default=null}
 CombinedDataTableRows {
  type="array<ST_Anim_Toolbox_RowComposite>"
  default [{
    rowName=None
    row {
     category=General
     description=no description provided...
     documentationURL=""
     icon=null
     nice Name=nicename
     pythonCommand=""
     pythonPath=""
     tool Action=Launch UI
     widget BP=null
    }
   }]
 }
 CombinedDataTableSources_List {type="array<string>", default []}
 CombinedDataTableSources_User {type="array<ST_Anim_Toolbox_SourceTag>", default []}
 CombinedRowNames {type="array<string>", default []}
 Data Table Struct {
  type=ST_Anim_Toolbox_Widget_Tab
  default {
   category=User
   description=no description provided...
   documentationURL=""
   icon=null
   nice Name=nicename
   pythonCommand=""
   pythonPath=""
   tool Action=Launch UI
   widget BP=null
  }
 }
 DataTablePath {type=string, default=""}
 Debug_HiddenTools {type=string, default=""}
 HiddenToolRows {type="array<string>", default []}
 LocPythonCommand {type=string, default=""}
 LocPythonPath {type=string, default=""}
 MatchedRowNames {type="array<string>", default []}
 ProjectDataTable {type=DataTable, default=null}
 Setting_ShowDescriptions {type=boolean}
 Setting_ShowDocLinks {type=boolean}
 Setting_ShowIcons {type=boolean}
 Setting_ToolList {type=boolean}
 Temp_CheckChangedName {type=string, default=""}
 Temp_CheckChangedState {type=boolean}
 TempTableRow {
  type=ST_Anim_Toolbox_Widget_Tab
  default {
   category=User
   description=no description provided...
   documentationURL=""
   icon=null
   nice Name=nicename
   pythonCommand=""
   pythonPath=""
   tool Action=Launch UI
   widget BP=null
  }
 }
 ToolboxDataTable {type=DataTable, default=&/DT_Toolbox_Tabs}
 UserDataTable {type=DataTable, default=null}
]
widgets {
 info {rootWidgetClass=&CanvasPanel, widgetCount=12}
 widgets [
  CanvasPanel_76 {widgetClassPath=&CanvasPanel, widget=CanvasPanel_76, properties {clipping=ClipToBounds}}
  VerticalBox_320 {
   widgetClassPath=&VerticalBox
   parent=CanvasPanel_76
   widget=VerticalBox_320
   slot {class=CanvasPanelSlot, properties {layoutData {alignment=<0 0>, anchors {maximum=<1 1>, minimum=<0 0>}, offsets=<ltrb 0 0 0 0>}}}
  }
  HorizontalBox_74 {
   widgetClassPath=&HorizontalBox
   parent=VerticalBox_320
   widget=HorizontalBox_74
   properties {clipping=ClipToBounds}
   slot {class=VerticalBoxSlot, properties {verticalAlignment=VAlign_Top, padding=<ltrb 4 2 4 0>}}
  }
  TextBlock_105 {
   widgetClassPath=&TextBlock
   parent=HorizontalBox_74
   widget=TextBlock_105
   properties {
    text=Anim Toolbox
    colorAndOpacity {colorUseRule=UseColor_Specified, specifiedColor=<rgba 0.8963 0.8963 0.8963 0.5>}
    font {
     bForceMonospaced=false
     bMaterialIsStencil=false
     fontMaterial=null
     fontObject=&Roboto
     letterSpacing=0
     monospacedWidth=0
     size=12
     skewAmount=0
     typefaceFontName=Bold
     outlineSettings {
      bApplyOutlineToDropShadows=false
      bMiteredCorners=false
      bSeparateFillAlpha=false
      outlineMaterial=null
      outlineSize=0
      outlineColor=<rgba 0 0 0 1>
     }
    }
   }
   slot {
    class=HorizontalBoxSlot
    properties {horizontalAlignment=HAlign_Center, verticalAlignment=VAlign_Bottom, padding=<ltrb 5 10 0 0>, size {sizeRule=Fill, value=1}}
   }
  }
  SettingsButton {
   widgetClassPath=&Button
   parent=HorizontalBox_74
   widget=SettingsButton
   properties {
    toolTipText=Settings Menu
    widgetStyle {
     clickedSlateSound {resourceObject=null}
     disabled {
      bIsDynamicallyLoaded=false
      drawAs=RoundedBox
      imageType=NoImage
      mirroring=NoMirror
      resourceName=None
      resourceObject=null
      tiling=NoTile
      imageSize=<0 0>
      margin=<ltrb 0 0 0 0>
      outlineSettings {
       bUseBrushTransparency=false
       roundingType=FixedRadius
       width=1
       color {colorUseRule=UseColor_ColorTable, specifiedColor=<rgba 0 0 0 0>}
       cornerRadii=<4 4 4 4>
      }
      tintColor {colorUseRule=UseColor_Specified, specifiedColor=<rgba 0.0395 0.0395 0.0395 1>}
      uVRegion {bIsValid=false, max=<0 0>, min=<0 0>}
     }
     disabledForeground {colorUseRule=UseColor_ColorTable, specifiedColor=<rgba 0 0 0 0>}
     hovered {
      bIsDynamicallyLoaded=false
      drawAs=RoundedBox
      imageType=FullColor
      mirroring=NoMirror
      resourceName=None
      resourceObject=null
      tiling=NoTile
      imageSize=<28 28>
      margin=<ltrb 0 0 0 0>
      outlineSettings {
       bUseBrushTransparency=false
       roundingType=FixedRadius
       width=0
       color {colorUseRule=UseColor_ColorTable, specifiedColor=<rgba 0 0 0 0>}
       cornerRadii=<4 4 4 4>
      }
      tintColor {colorUseRule=UseColor_Specified, specifiedColor=<rgba 0.0941 0.0941 0.0941 1>}
      uVRegion {bIsValid=false, max=<0 0>, min=<0 0>}
     }
     hoveredForeground {colorUseRule=UseColor_ColorTable, specifiedColor=<rgba 0 0 0 0>}
     hoveredSlateSound {resourceObject=null}
     normal {
      bIsDynamicallyLoaded=false
      drawAs=NoDrawType
      imageType=FullColor
      mirroring=NoMirror
      resourceName=None
      resourceObject=null
      tiling=NoTile
      imageSize=<28 28>
      margin=<ltrb 0 0 0 0>
      outlineSettings {
       bUseBrushTransparency=false
       roundingType=FixedRadius
       width=0
       color {colorUseRule=UseColor_ColorTable, specifiedColor=<rgba 0 0 0 0>}
       cornerRadii=<4 4 4 4>
      }
      tintColor {colorUseRule=UseColor_Specified, specifiedColor=<rgba 0.0176 0.0176 0.0176 1>}
      uVRegion {bIsValid=false, max=<0 0>, min=<0 0>}
     }
     normalForeground {colorUseRule=UseColor_ColorTable, specifiedColor=<rgba 0 0 0 0>}
     normalPadding=<ltrb 0 0 0 0>
     pressed {
      bIsDynamicallyLoaded=false
      drawAs=RoundedBox
      imageType=FullColor
      mirroring=NoMirror
      resourceName=None
      resourceObject=null
      tiling=NoTile
      imageSize=<28 28>
      margin=<ltrb 0 0 0 0>
      outlineSettings {
       bUseBrushTransparency=false
       roundingType=FixedRadius
       width=0
       color {colorUseRule=UseColor_ColorTable, specifiedColor=<rgba 0 0 0 0>}
       cornerRadii=<4 4 4 4>
      }
      tintColor {colorUseRule=UseColor_Specified, specifiedColor=<rgba 0.0392 0.0392 0.0392 1>}
      uVRegion {bIsValid=false, max=<0 0>, min=<0 0>}
     }
     pressedForeground {colorUseRule=UseColor_ColorTable, specifiedColor=<rgba 0 0 0 0>}
     pressedPadding=<ltrb 0 0 0 0>
     pressedSlateSound {resourceObject=null}
    }
   }
   slot {class=HorizontalBoxSlot, properties {horizontalAlignment=HAlign_Right, verticalAlignment=VAlign_Bottom, padding=<ltrb 0 0 16 0>}}
  }
  HorizontalBox_5 {
   widgetClassPath=&HorizontalBox
   parent=SettingsButton
   widget=HorizontalBox_5
   slot {class=ButtonSlot, properties {padding=<ltrb 0 0 0 0>}}
  }
  AssetImage_2 {
   widgetClassPath=&Image
   bIsVariable=true
   parent=HorizontalBox_5
   widget=AssetImage_2
   properties {brush {
     bIsDynamicallyLoaded=false
     drawAs=Image
     imageType=FullColor
     mirroring=NoMirror
     resourceName=None
     resourceObject=/Game/ArtTools/Tech_Art/InstancedMeshesCluster/Icon/Icon_Visible
     tiling=NoTile
     imageSize=<16 16>
     margin=<ltrb 0 0 0 0>
     outlineSettings {
      bUseBrushTransparency=false
      roundingType=HalfHeightRadius
      width=0
      color {colorUseRule=UseColor_Specified, specifiedColor=<rgba 0 0 0 0>}
      cornerRadii=<0 0 0 0>
     }
     tintColor {colorUseRule=UseColor_Specified, specifiedColor=<rgba 1 1 1 1>}
     uVRegion {bIsValid=false, max=<0 0>, min=<0 0>}
    }}
   slot {class=HorizontalBoxSlot, properties {horizontalAlignment=HAlign_Center, verticalAlignment=VAlign_Center, padding=<ltrb 3 3 3 3>}}
  }
  SB_ButtonScroll {
   widgetClassPath=&ScrollBox
   parent=VerticalBox_320
   widget=SB_ButtonScroll
   properties {
    alwaysShowScrollbarTrack=true
    widgetBarStyle {
     thickness=8
     draggedThumbImage {
      bIsDynamicallyLoaded=false
      drawAs=RoundedBox
      imageType=NoImage
      mirroring=NoMirror
      resourceName=None
      resourceObject=null
      tiling=NoTile
      imageSize=<0 0>
      margin=<ltrb 0 0 0 0>
      outlineSettings {
       bUseBrushTransparency=false
       roundingType=FixedRadius
       width=0
       color {colorUseRule=UseColor_Specified, specifiedColor=<rgba 0 0 0 0>}
       cornerRadii=<4 4 4 4>
      }
      tintColor {colorUseRule=UseColor_Specified, specifiedColor=<rgba 0.2159 0.2159 0.2159 1>}
      uVRegion {bIsValid=false, max=<0 0>, min=<0 0>}
     }
     horizontalBackgroundImage {
      bIsDynamicallyLoaded=false
      drawAs=NoDrawType
      imageType=NoImage
      mirroring=NoMirror
      resourceName=None
      resourceObject=null
      tiling=NoTile
      imageSize=<0 0>
      margin=<ltrb 0 0 0 0>
      outlineSettings {
       bUseBrushTransparency=false
       roundingType=HalfHeightRadius
       width=0
       color {colorUseRule=UseColor_Specified, specifiedColor=<rgba 0 0 0 0>}
       cornerRadii=<0 0 0 0>
      }
      tintColor {colorUseRule=UseColor_Specified, specifiedColor=<rgba 1 1 1 1>}
      uVRegion {bIsValid=false, max=<0 0>, min=<0 0>}
     }
     horizontalBottomSlotImage {
      bIsDynamicallyLoaded=false
      drawAs=NoDrawType
      imageType=NoImage
      mirroring=NoMirror
      resourceName=None
      resourceObject=null
      tiling=NoTile
      imageSize=<0 0>
      margin=<ltrb 0 0 0 0>
      outlineSettings {
       bUseBrushTransparency=false
       roundingType=HalfHeightRadius
       width=0
       color {colorUseRule=UseColor_Specified, specifiedColor=<rgba 0 0 0 0>}
       cornerRadii=<0 0 0 0>
      }
      tintColor {colorUseRule=UseColor_Specified, specifiedColor=<rgba 1 1 1 1>}
      uVRegion {bIsValid=false, max=<0 0>, min=<0 0>}
     }
     horizontalTopSlotImage {
      bIsDynamicallyLoaded=false
      drawAs=NoDrawType
      imageType=NoImage
      mirroring=NoMirror
      resourceName=None
      resourceObject=null
      tiling=NoTile
      imageSize=<0 0>
      margin=<ltrb 0 0 0 0>
      outlineSettings {
       bUseBrushTransparency=false
       roundingType=HalfHeightRadius
       width=0
       color {colorUseRule=UseColor_Specified, specifiedColor=<rgba 0 0 0 0>}
       cornerRadii=<0 0 0 0>
      }
      tintColor {colorUseRule=UseColor_Specified, specifiedColor=<rgba 1 1 1 1>}
      uVRegion {bIsValid=false, max=<0 0>, min=<0 0>}
     }
     hoveredThumbImage {
      bIsDynamicallyLoaded=false
      drawAs=RoundedBox
      imageType=NoImage
      mirroring=NoMirror
      resourceName=None
      resourceObject=null
      tiling=NoTile
      imageSize=<0 0>
      margin=<ltrb 0 0 0 0>
      outlineSettings {
       bUseBrushTransparency=false
       roundingType=FixedRadius
       width=0
       color {colorUseRule=UseColor_Specified, specifiedColor=<rgba 0 0 0 0>}
       cornerRadii=<4 4 4 4>
      }
      tintColor {colorUseRule=UseColor_Specified, specifiedColor=<rgba 0.2159 0.2159 0.2159 1>}
      uVRegion {bIsValid=false, max=<0 0>, min=<0 0>}
     }
     normalThumbImage {
      bIsDynamicallyLoaded=false
      drawAs=RoundedBox
      imageType=NoImage
      mirroring=NoMirror
      resourceName=None
      resourceObject=null
      tiling=NoTile
      imageSize=<0 0>
      margin=<ltrb 0 0 0 0>
      outlineSettings {
       bUseBrushTransparency=false
       roundingType=FixedRadius
       width=0
       color {colorUseRule=UseColor_Specified, specifiedColor=<rgba 0 0 0 0>}
       cornerRadii=<4 4 4 4>
      }
      tintColor {colorUseRule=UseColor_Specified, specifiedColor=<rgba 0.0953 0.0953 0.0953 1>}
      uVRegion {bIsValid=false, max=<0 0>, min=<0 0>}
     }
     verticalBackgroundImage {
      bIsDynamicallyLoaded=false
      drawAs=NoDrawType
      imageType=NoImage
      mirroring=NoMirror
      resourceName=None
      resourceObject=null
      tiling=NoTile
      imageSize=<0 0>
      margin=<ltrb 0 0 0 0>
      outlineSettings {
       bUseBrushTransparency=false
       roundingType=HalfHeightRadius
       width=0
       color {colorUseRule=UseColor_Specified, specifiedColor=<rgba 0 0 0 0>}
       cornerRadii=<0 0 0 0>
      }
      tintColor {colorUseRule=UseColor_Specified, specifiedColor=<rgba 1 1 1 1>}
      uVRegion {bIsValid=false, max=<0 0>, min=<0 0>}
     }
     verticalBottomSlotImage {
      bIsDynamicallyLoaded=false
      drawAs=NoDrawType
      imageType=NoImage
      mirroring=NoMirror
      resourceName=None
      resourceObject=null
      tiling=NoTile
      imageSize=<0 0>
      margin=<ltrb 0 0 0 0>
      outlineSettings {
       bUseBrushTransparency=false
       roundingType=HalfHeightRadius
       width=0
       color {colorUseRule=UseColor_Specified, specifiedColor=<rgba 0 0 0 0>}
       cornerRadii=<0 0 0 0>
      }
      tintColor {colorUseRule=UseColor_Specified, specifiedColor=<rgba 1 1 1 1>}
      uVRegion {bIsValid=false, max=<0 0>, min=<0 0>}
     }
     verticalTopSlotImage {
      bIsDynamicallyLoaded=false
      drawAs=NoDrawType
      imageType=NoImage
      mirroring=NoMirror
      resourceName=None
      resourceObject=null
      tiling=NoTile
      imageSize=<0 0>
      margin=<ltrb 0 0 0 0>
      outlineSettings {
       bUseBrushTransparency=false
       roundingType=HalfHeightRadius
       width=0
       color {colorUseRule=UseColor_Specified, specifiedColor=<rgba 0 0 0 0>}
       cornerRadii=<0 0 0 0>
      }
      tintColor {colorUseRule=UseColor_Specified, specifiedColor=<rgba 1 1 1 1>}
      uVRegion {bIsValid=false, max=<0 0>, min=<0 0>}
     }
    }
    widgetStyle {
     barThickness=8
     bottomShadowBrush {
      bIsDynamicallyLoaded=false
      drawAs=Box
      imageType=FullColor
      mirroring=NoMirror
      resourceName=../../../Engine/Content/Slate/Common/ScrollBoxShadowBottom.png
      resourceObject=null
      tiling=NoTile
      imageSize=<16 8>
      margin=<ltrb 0.5 0 0.5 1>
      outlineSettings {
       bUseBrushTransparency=false
       roundingType=HalfHeightRadius
       width=0
       color {colorUseRule=UseColor_Specified, specifiedColor=<rgba 0 0 0 0>}
       cornerRadii=<0 0 0 0>
      }
      tintColor {colorUseRule=UseColor_Specified, specifiedColor=<rgba 1 1 1 1>}
      uVRegion {bIsValid=false, max=<0 0>, min=<0 0>}
     }
     horizontalScrolledContentPadding=<ltrb 0 0 1 0>
     leftShadowBrush {
      bIsDynamicallyLoaded=false
      drawAs=Box
      imageType=FullColor
      mirroring=NoMirror
      resourceName=../../../Engine/Content/Slate/Common/ScrollBoxShadowLeft.png
      resourceObject=null
      tiling=NoTile
      imageSize=<8 16>
      margin=<ltrb 1 0.5 0 0.5>
      outlineSettings {
       bUseBrushTransparency=false
       roundingType=HalfHeightRadius
       width=0
       color {colorUseRule=UseColor_Specified, specifiedColor=<rgba 0 0 0 0>}
       cornerRadii=<0 0 0 0>
      }
      tintColor {colorUseRule=UseColor_Specified, specifiedColor=<rgba 1 1 1 1>}
      uVRegion {bIsValid=false, max=<0 0>, min=<0 0>}
     }
     rightShadowBrush {
      bIsDynamicallyLoaded=false
      drawAs=Box
      imageType=FullColor
      mirroring=NoMirror
      resourceName=../../../Engine/Content/Slate/Common/ScrollBoxShadowRight.png
      resourceObject=null
      tiling=NoTile
      imageSize=<8 16>
      margin=<ltrb 0 0.5 1 0.5>
      outlineSettings {
       bUseBrushTransparency=false
       roundingType=HalfHeightRadius
       width=0
       color {colorUseRule=UseColor_Specified, specifiedColor=<rgba 0 0 0 0>}
       cornerRadii=<0 0 0 0>
      }
      tintColor {colorUseRule=UseColor_Specified, specifiedColor=<rgba 1 1 1 1>}
      uVRegion {bIsValid=false, max=<0 0>, min=<0 0>}
     }
     topShadowBrush {
      bIsDynamicallyLoaded=false
      drawAs=Box
      imageType=FullColor
      mirroring=NoMirror
      resourceName=../../../Engine/Content/Slate/Common/ScrollBoxShadowTop.png
      resourceObject=null
      tiling=NoTile
      imageSize=<16 8>
      margin=<ltrb 0.5 1 0.5 0>
      outlineSettings {
       bUseBrushTransparency=false
       roundingType=HalfHeightRadius
       width=0
       color {colorUseRule=UseColor_Specified, specifiedColor=<rgba 0 0 0 0>}
       cornerRadii=<0 0 0 0>
      }
      tintColor {colorUseRule=UseColor_Specified, specifiedColor=<rgba 1 1 1 1>}
      uVRegion {bIsValid=false, max=<0 0>, min=<0 0>}
     }
     verticalScrolledContentPadding=<ltrb 0 0 0 1>
    }
   }
   slot {class=VerticalBoxSlot, properties {padding=<ltrb 6 0 6 4>, size {sizeRule=Fill, value=1}}}
  }
  ButtonListBox {
   widgetClassPath=&VerticalBox
   bIsVariable=true
   parent=SB_ButtonScroll
   widget=ButtonListBox
   slot {class=ScrollBoxSlot, properties {size {sizeRule=Fill, value=1}}}
  }
  Overlay_67 {
   widgetClassPath=&Overlay
   parent=CanvasPanel_76
   widget=Overlay_67
   properties {clipping=ClipToBounds}
   slot {class=CanvasPanelSlot, properties {zOrder=1, layoutData {alignment=<0 0>, anchors {maximum=<1 1>, minimum=<0 0>}, offsets=<ltrb 0 0 0 0>}}}
  }
  ModalLayer {
   widgetClassPath=&NamedSlot
   bIsVariable=true
   parent=Overlay_67
   widget=ModalLayer
   properties {bExposeOnInstanceOnly=true}
   slot {class=OverlaySlot, properties {horizontalAlignment=HAlign_Fill, verticalAlignment=VAlign_Fill}}
  }
  EUW_ContextMenu {
   widgetClassPath=/AnimInEdTools/ContextMenu/EUW_ContextMenu
   bIsVariable=true
   parent=Overlay_67
   widget=EUW_ContextMenu
   properties {visibility=Collapsed}
   slot {class=OverlaySlot}
  }
 ]
}
