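# Editor Utility Widget that serves as a shortcut to the Anim In Engine chatbot on Slack, showing a brief "opening" message.
# On Construct it runs Python to open the Slack URL in a web browser, waits one frame, then runs Python that closes its own tab through EditorUtilitySubsystem.
# An unconnected LaunchURL node holds the same URL, and the other events are empty.
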EditorUtilityWidgetBlueprint /AnimInEdTools/AnimationToolbox/EUW_AskAnim
graphs [EventGraph {nodes [
   Event Construct()
   > Execute Python Script(PythonScript="import webbrowser\r\n\r\nwebbrowser.open(\"https://epic.enterprise.slack.com/team/U092NLQ9883\")")
   > DelayUntilNextFrame!(execute, WorldContextObject:Object Reference) -> (then)
   > Execute Python Script(PythonScript="import unreal\r\n\r\neus = unreal.get_editor_subsystem(unreal.EditorUtilitySubsystem)\r\nblueprint = unreal.load_object(None, '&self')\r\ntab_id = eus.get_tab_id_from_blueprint(blueprint)\r\neus.close_tab_by_id(tab_id)\r\n")
   LaunchURL!(execute, URL:String=https://epic.enterprise.slack.com/team/U092NLQ9883) -> (then)
   Event PreConstruct()
   Event Tick()
   Event Run()
  ]}]
variables []
widgets {
 info {rootWidgetClass=&StackBox, widgetCount=2}
 widgets [
  StackBox_0 {widgetClassPath=&StackBox, widget=StackBox_0}
  TextBlock_88 {
   widgetClassPath=&TextBlock
   parent=StackBox_0
   widget=TextBlock_88
   properties {text=Opening Anim In Engine Chatbot...}
   slot {class=StackBoxSlot, properties {horizontalAlignment=HAlign_Center, verticalAlignment=VAlign_Center, size {sizeRule=Fill, value=1}}}
  }
 ]
}
