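# Unlit debug material for AnimToTexture that visualizes per-vertex data from the bone weight texture.
# A Custom node offsets TexCoord function UVs by frame rows, with frame fixed at zero, to sample the texture.
# Static switches pick one RGBA channel as the influence, defaulting to red.
# Vertex interpolators pass it to the pixel stage, where ShowBoneIndex selects the raw value or a floored 0-255 bone index.
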Material /AnimToTexture/Materials/M_BoneIndicesDebug
graph {
 properties {pixelDepthOffsetMode=PDOM_Legacy, shadingModel=MSM_Unlit, usedShadingModels=Unlit}
 nodes [
  MaterialExpressionMaterialFunctionCall_3 = MaterialFunctionCall() {materialFunction=/AnimToTexture/Materials/MaterialFunctions/TexCoord}
  MaterialExpressionConstant_0 = Constant()
  n1 = TextureObjectParameter() {parameterName=BoneWeightTexture}
  MaterialExpressionTextureProperty_0 = TextureProperty(Tex=@n1)
  MaterialExpressionScalarParameter_2 = ScalarParameter() {defaultValue=1, parameterName=BoneWeightsRowsPerFrame}
  n2 = Custom!(
   UVs=@MaterialExpressionMaterialFunctionCall_3.Result
   Frame=@MaterialExpressionConstant_0
   TextureSize=@MaterialExpressionTextureProperty_0
   RowsPerFrame=@MaterialExpressionScalarParameter_2
  ) {
   code="UVs.y += ((int(Frame) * int(RowsPerFrame) / TextureSize.y)) % TextureSize.y;\r\n\r\nreturn UVs;"
   desc=This can be done with Math nodes
   description=OffsetUVs
   outputType=CMOT_Float2
   inputs [
    {inputName=UVs, input {expression=&self:MaterialExpressionMaterialFunctionCall_3}}
    {inputName=Frame, input {expression=&self:MaterialExpressionConstant_0}}
    {inputName=TextureSize, input {expression=&self:MaterialExpressionTextureProperty_0}}
    {inputName=RowsPerFrame, input {expression=&self:MaterialExpressionScalarParameter_2}}
   ]
  }
  n3 = MaterialFunctionCall(Float4=TextureSample(UVs=@n2, Tex=@n1) {samplerType=SAMPLERTYPE_LinearColor, texture=/AnimToTexture/Characters/Mannequin/Textures/BoneAnimation/TX_BoneWeight}.RGBA) {materialFunction=&BreakOutFloat4Components}
  n4 = StaticSwitchParameter(True=@n3.A, False=@n3.A) {parameterName=UseInfluence3}
  n5 = StaticSwitchParameter(
   True=@n3.R
   False=StaticSwitchParameter(True=@n3.G, False=StaticSwitchParameter(True=@n3.B, False=@n4) {parameterName=UseInfluence2}) {parameterName=UseInfluence1}
  ) {defaultValue=true, parameterName=UseInfluence0}
  n6 = Multiply(A=@n5) {constB=255}
  StaticSwitchParameter(True=Floor(None=VertexInterpolator(VS=@n6).PS), False=VertexInterpolator(VS=@n5).PS) {parameterName=ShowBoneIndex}
 ]
}
