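# BP_EventHub is an Editor Utility Blueprint in the AnimationLibrary that acts as a generic event hub.
# It exposes one OnEvent dispatcher whose signature carries an EventName and an Object Payload.
# The Emit function broadcasts that dispatcher with the name and payload it receives.
# It has no filtering or stored state beyond the dispatcher, so every bound listener receives every event.
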EditorUtilityBlueprint /AnimationLibrary/BP/BP_EventHub
graphs [
 Emit {nodes [
   n1 = Emit!() -> (then, EventName:Name, Payload:Object Reference) {is=title K2Node_FunctionEntry}
   > Call On Event!(execute:Exec, self:BP Event Hub Object Reference, EventName:Name=@n1, Payload:Object Reference=@n1.Payload) -> (then:Exec) {is=title K2Node_CallDelegate}
  ]}
 EventGraph {links [], nodes []}
 OnEvent {nodes [OnEvent!() -> (then, EventName:Name, Payload:Object Reference) {is=title K2Node_FunctionEntry}]}
]
variables [OnEvent {default=()}]
