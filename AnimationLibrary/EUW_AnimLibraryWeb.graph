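# Editor utility widget that hosts the AnimLibrary tool's web UI in an embedded browser, backed by a Python-launched local server.
# On init it purges cached AnimLibrary Python modules, starts the web launcher and loads its URL with a timestamp cache-buster; Destruct stops the launcher.
# A Refresh button, revealed by a Shift key press, stops and restarts this chain for hot reloading.
# Browser console messages are forwarded to the log.
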EditorUtilityWidgetBlueprint /AnimationLibrary/EUW_AnimLibraryWeb
graphs [
 EventGraph {nodes [
   n1 = "On Console Message (wb_main)"!() -> (OutputDelegate, then, Message:String, Source:String, Line:int=0) {is=title K2Node_ComponentBoundEvent}
   > LogString(InString=Append!(A="File: ", B=@n1.Source, C:String="\r\nMessage:\r\n", D:String=@n1.Message) -> (ReturnValue))
   "On Clicked (btn_reload)"() {is=title K2Node_ComponentBoundEvent}
   > n2 = Execute Python Script(PythonScript="from AnimLibrary import web_launcher\nweb_launcher.stop()")
   Execute Python Script(
    execute=[@n4.then, @n2.then]
    PythonScript="import sys\r\n\r\ndef clean_anim():\r\n    for m in list(sys.modules.keys()):\r\n        if 'AnimLibrary' not in m:\r\n            continue\r\n        del sys.modules[m]\r\n\r\nclean_anim()"
   )
   > n3 = Execute Python Script!(
    execute
    PythonScript="import time\r\nfrom AnimLibrary import web_launcher\r\nurl = web_launcher.start()\r\nt = str(int(time.time()))\r\nurl = f'{url}/?v={t}'\r\n\r\n\r\n\r\n\r\n"
   ) -> (then, ReturnValue, url:String) {outputs [url]}
   > LoadURL!(execute, self:Web Browser Object Reference=Get wb_main() -> (wb_main:Web Browser Object Reference), NewURL:String=@n3.url) -> (then)
   n4 = Event OnInitialized()
   Event Destruct()
   > Execute Python Script!(execute, PythonScript="from AnimLibrary import web_launcher\nweb_launcher.stop()") -> (then, ReturnValue)
   Event PreConstruct()
   Event Construct()
   Event Tick()
  ]}
 OnPreviewKeyDown {nodes [
   n1 = OnPreviewKeyDown!() -> (then, MyGeometry:Geometry Structure, InKeyEvent:Key Event Structure) {is=title K2Node_FunctionEntry}
   n2 = GetInputEventFromKeyEvent!(Event:"Key Event Structure (by ref)"=@n1.InKeyEvent) -> (ReturnValue:Input Event Structure)
   > SetVisibility(
    self=Get btn_reload() -> (btn_reload:Editor Utility Button Object Reference)
    InVisibility=Select(Option 0=Collapsed, Option 1=Visible, Index=Is Shift Down!(Input:"Input Event Structure (by ref)"=@n2) -> (ReturnValue:bool))
   )
   > Return Node!(execute, ReturnValue:Event Reply Structure=Unhandled())
  ]}
]
variables []
widgets {
 info {rootWidgetClass=&StackBox, widgetCount=4}
 widgets [
  Container {widgetClassPath=&StackBox, bIsVariable=true, widget=Container, properties {orientation=Orient_Vertical}}
  wb_main {
   widgetClassPath=/Script/WebBrowserWidget.WebBrowser
   bIsVariable=true
   parent=Container
   widget=wb_main
   properties {initialURL=http://127.0.0.1:8765}
   slot {class=StackBoxSlot, properties {size {sizeRule=Fill, value=1}}}
  }
  btn_reload {
   widgetClassPath=&EditorUtilityButton
   parent=Container
   widget=btn_reload
   properties {visibility=Collapsed}
   slot {class=StackBoxSlot}
  }
  TextBlock_54 {widgetClassPath=&TextBlock, parent=btn_reload, widget=TextBlock_54, properties {text=Refresh}, slot {class=ButtonSlot}}
 ]
}
