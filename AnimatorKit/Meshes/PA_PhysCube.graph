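# PA_PhysCube is a physics asset in the AnimatorKit meshes folder that gives a cube mesh a simple collision body for physics.
# It holds a single body named box with one box primitive of equal size on all three axes, rotated a quarter turn.
# It defines no constraints, so the body has no joints to other bodies.
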PhysicsAsset /AnimatorKit/Meshes/PA_PhysCube
bodies {box [{shapeType=Box, extentX=101, extentY=101, extentZ=101, rotation=<rot 0 -90 0>}]}
constraints []
