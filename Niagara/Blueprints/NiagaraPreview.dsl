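; Niagara preview actor Blueprint that displays one Niagara system alongside X-axis and Y-axis text labels.
; The SetSystem event assigns the given system to the Effect component.
; The SetLabelText event writes the two given texts to the XLabel and YLabel text render components.
; BeginPlay, actor overlap and the construction script are empty.
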(event EventBeginPlay)

(event Collision|EventActorBeginOverlap
  (params (OtherActor :type Actor%20Object%20Reference)))

(event Niagara|Preview|EventSetSystem
  (params (InSystem :type Niagara%20System%20Object%20Reference))
  (Niagara|SetNiagaraSystemAsset (Variables|Default|GetEffect) InSystem true))

(event Niagara|Preview|EventSetLabelText
  (params (InXAxisText :type Text%20%28by%20ref%29) (InYAxisText :type Text%20%28by%20ref%29))
  (Rendering|Components|TextRender|SetText (Variables|Default|GetXLabel) InXAxisText)
  (Rendering|Components|TextRender|SetText (Variables|Default|GetYLabel) InYAxisText))

(fn ConstructionScript)
