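; Niagara preview grid axis that varies LOD distance across preview components using a Distances array, with Num reporting the entry count.
; ApplyToPreview picks the distance by clamped index, sets it as the component's preview LOD distance, and returns a text label.
; When bApplyRealDistances is set, it also moves the component to that distance on relative X or Y, according to bIsXAxis.
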(fn ApplyToPreview
  (params (PreviewComponent :type Niagara%20Particle%20System%20Component%20Object%20Reference) (PreviewIndex :type Integer) (bIsXAxis :type Boolean))
  (bind _distances (Variables|Default|GetDistances))
  (bind _returnvalue (Utilities|Array|Length _distances))
  (bind _returnvalue_2 (Math|Integer|Clamp(Integer) PreviewIndex 0 _returnvalue))
  (bind _output (Utilities|Array|Get(acopy) _distances _returnvalue_2))
  (if (Variables|Default|GetbApplyRealDistances)
    (Transformation|SetRelativeLocation PreviewComponent (select bIsXAxis (Math|Vector|MakeVector _output (.y (.location (Transformation|GetRelativeTransform PreviewComponent))) (.z (.location (Transformation|GetRelativeTransform PreviewComponent)))) (Math|Vector|MakeVector (.x (.location (Transformation|GetRelativeTransform PreviewComponent))) _output (.z (.location (Transformation|GetRelativeTransform PreviewComponent))))) false false)
    (else
      _))
  (Preview|SetPreviewLODDistance PreviewComponent true _output 0.0)
  (return (Utilities|String|Append "LODDistance = " (Utilities|String|ToString(Float) _output))))

(fn Num
  (return (Utilities|Array|Length (Variables|Default|GetDistances))))
