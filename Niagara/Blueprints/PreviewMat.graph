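# PreviewMat is a minimal Default Lit material in the Niagara Blueprints folder, apparently for preview use.
# Its graph holds a single Constant3Vector set to white, with no other expressions and no material input connection shown.
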Material /Niagara/Blueprints/PreviewMat
graph {properties {pixelDepthOffsetMode=PDOM_Legacy, usedShadingModels=Default Lit}, nodes [Constant3Vector() {constant=<rgba 1 1 1 0>}]}
