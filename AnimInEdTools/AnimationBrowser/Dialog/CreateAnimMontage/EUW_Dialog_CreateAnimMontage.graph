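# Editor utility dialog that collects and validates inputs for creating an AnimMontage from a Level Sequence's linked AnimSequences.
# Set Level Sequence fills bindings with Control Rig or animation tracks, BlankMontage-derived presets from the Asset Registry, and the selected binding's linked AnimSequences.
# Selecting a sequence suggests an "AM_" name.
# Delegates hand validation and creation to external listeners, while Set Data is Valid toggles the Create button and messages.
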EditorUtilityWidgetBlueprint /AnimInEdTools/AnimationBrowser/Dialog/CreateAnimMontage/EUW_Dialog_CreateAnimMontage
graphs [
 Call Validate User Data {nodes [
   Call Validate User Data() {is=title K2Node_FunctionEntry}
   n1 = Get UI Data!() -> (
    AnimMontage Name:String
    AnimSequence Path:String
    Montage Preset:Blank Montage Class Reference
    Proxy Binding:Movie Scene Binding Proxy Structure
    Valid:bool
   ) {is=title K2Node_CallFunction}
   > Branch(Condition=@n1.Valid)
   > Call Validate Data!(
    execute:Exec
    self:EUW Dialog Create Anim Montage Object Reference
    AnimMontage Name:String=@n1
    AnimSequence Path:String=@n1.AnimSequence Path
    Montage Preset:Blank Montage Class Reference=@n1.Montage Preset
    Binding Proxy:Movie Scene Binding Proxy Structure=@n1.Proxy Binding
    Widget:EUW Dialog Create Anim Montage Object Reference=Self-Reference()
   ) -> (then:Exec) {is=title K2Node_CallDelegate}
  ]}
 Cancel Button Pressed {nodes [Cancel Button Pressed() {is=title K2Node_FunctionEntry}]}
 Create Button Pressed {nodes [Create Button Pressed!() -> (
    then
    AnimMontage Name:String
    AnimSequence Path:String
    Montage Preset:Blank Montage Class Reference
    Binding Proxy:Movie Scene Binding Proxy Structure
   ) {is=title K2Node_FunctionEntry}]}
 EventGraph {nodes [
   "On Clicked (CancelButton)"() {is=title K2Node_ComponentBoundEvent}
   > Call Cancel Button Pressed!(execute:Exec, self:EUW Dialog Create Anim Montage Object Reference) -> (then:Exec) {is=title K2Node_CallDelegate}
   "On Clicked (CreateButton)"() {is=title K2Node_ComponentBoundEvent}
   n1 = Get UI Data()
   > Branch(Condition=@n1.Valid)
   > Call Create Button Pressed!(
    execute:Exec
    self:EUW Dialog Create Anim Montage Object Reference
    AnimMontage Name:String=@n1
    AnimSequence Path:String=@n1.AnimSequence Path
    Montage Preset:Blank Montage Class Reference=@n1.Montage Preset
    Binding Proxy:Movie Scene Binding Proxy Structure=@n1.Proxy Binding
   ) -> (then:Exec) {is=title K2Node_CallDelegate}
   "On Selection Changed (AssetComboBox)"!() -> (OutputDelegate, then, SelectedItem:String, SelectionType:ESelectInfo Enum) {is=title K2Node_ComponentBoundEvent}
   > Update AnimSequence ListView()
   > Call Validate User Data() {is=title K2Node_CallFunction}
   n2 = "On Item Clicked (ListView)"!() -> (OutputDelegate, then, Item:Object Reference) {is=title K2Node_ComponentBoundEvent}
   > Set Montage Name!(execute, self, Object:Object Reference=@n2.Item) -> (then) {is=title K2Node_CallFunction}
   > Call Validate User Data()
   "On Text Changed (AnimMontageName)"!() -> (OutputDelegate, then, Text:"Text (by ref)") {is=title K2Node_ComponentBoundEvent}
   > Call Validate User Data()
   n3 = Set Level Sequence!() -> (OutputDelegate, then, LevelSequence:Level Sequence Object Reference) {is=title K2Node_CustomEvent}
   > n4 = Set Level Sequence(Level Sequence:Level Sequence Object Reference=@n3.LevelSequence) -> (Output_Get:Level Sequence Object Reference) {is=title K2Node_VariableSet}
   > Update Bindings!(execute, self, LevelSequence:Level Sequence Object Reference=@n4) -> (then)
   > Update AnimMontage Presets()
   > Update AnimSequence ListView()
   > Call Validate User Data()
   Event PreConstruct()
   Event Construct()
   Event Tick()
  ]}
 Get UI Data {nodes [
   Get UI Data() {is=title K2Node_FunctionEntry}
   > Set Is Valid(Is Valid:bool=true) -> (Output_Get:bool)
   > n1 = Sequence!(execute) -> (then_0, then_1, then_2:Exec, then_3:Exec, then_4:Exec)
   n2 = GetSelectedIndex!(self:"Searchable Combo Box (String) Object Reference"=Get PresetComboBox() -> (PresetComboBox:"Searchable Combo Box (String) Object Reference")) -> (ReturnValue:int=0)
   n3 = Branch(execute=@n1.then_2, Condition="integer >= integer"(A=@n2))
   n4 = Find(TargetMap:Map of Integers to Strings=Get Montage Preset Map() -> (Montage Preset Map:Map of Integers to Strings), Key:"Integer (by ref)"=@n2)
   > n5 = Branch(Condition=@n4.ReturnValue)
   > n6 = LoadClassAsset_Blocking(AssetClass=To Soft Class Reference!(SoftClassPath_PathString:String=@n4) -> (ReturnValue))
   > n7 = Cast To BlankMontage Class!(execute:Exec, Class:Object Class Reference=@n6) -> (then:Exec, CastFailed:Exec, AsBlank Montage:Blank Montage Class Reference) {is=title K2Node_ClassDynamicCast}
   > Set Selected Montage Preset(Selected Montage Preset:Blank Montage Class Reference=@n7) -> (Output_Get:Blank Montage Class Reference)
   Set Is Valid(execute=[@n7.CastFailed, @n5.else, @n3.else], Is Valid=false)
   Return Node!(
    execute=@n1.then_4
    AnimMontage Name:String=Get AnimMontage() -> (AnimMontage:String)
    AnimSequence Path:String=Get AnimSequence() -> (AnimSequence:String)
    Montage Preset:Blank Montage Class Reference=Get Selected Montage Preset() -> (Selected Montage Preset:Blank Montage Class Reference)
    Proxy Binding:Movie Scene Binding Proxy Structure=Get Selected Proxy Binding() -> (Selected Proxy Binding:Movie Scene Binding Proxy Structure)
    Valid:bool=Get Is Valid() -> (Is Valid:bool)
   )
   n8 = "To String (Text)"(InText=Get Text(self=Get AnimMontageName() -> (AnimMontageName:Editor Utility Editable Text Box Object Reference)))
   n9 = Branch(execute=@n1.then_0, Condition=IsEmpty(InString=@n8))
   > Set Is Valid(Is Valid=false)
   Set AnimMontage(execute=@n9.else, AnimMontage:String=@n8) -> (Output_Get:String)
   n10 = Get Selected Item!(self:List View Object Reference=Get ListView() -> (ListView:Editor Utility List View Object Reference)) -> (ReturnValue:Object Reference) {is=title K2Node_CallFunction}
   n11 = Is Valid(exec=@n1.then_1, InputObject=@n10)
   > n12 = Cast To CreateAnimMontage_Object!(execute, Object=@n10) -> (then, CastFailed, AsCreate Anim Montage Object:Create Anim Montage Object Object Reference) {is=title K2Node_DynamicCast}
   > Set AnimSequence(AnimSequence:String=Get AnimSequence Path!(self:Create Anim Montage Object Object Reference=@n12) -> (AnimSequence Path:String)) -> (Output_Get:String)
   Set Is Valid(execute=[@n12.CastFailed, @n11.Is Not Valid], Is Valid=false)
   n13 = Find(
    TargetMap:Map of Integers to Movie Scene Binding Proxy Structures=Get Asset Binding Map() -> (Asset Binding Map:Map of Integers to Movie Scene Binding Proxy Structures)
    Key=GetSelectedIndex(self=Get AssetComboBox() -> (AssetComboBox:"Searchable Combo Box (String) Object Reference")) -> (ReturnValue=0)
   ) -> (Value:Movie Scene Binding Proxy Structure)
   n14 = Branch(execute=@n1.then_3, Condition=@n13.ReturnValue)
   > Set Selected Proxy Binding(Selected Proxy Binding:Movie Scene Binding Proxy Structure=@n13) -> (Output_Get:Movie Scene Binding Proxy Structure)
   Set Is Valid(execute=@n14.else, Is Valid=false)
  ]}
 Set Data is Valid {nodes [
   n1 = Set Data is Valid!() -> (then, Is Valid:bool=~, Message:String, Error:String) {is=title K2Node_FunctionEntry}
   > n2 = Sequence!(execute) -> (then_0, then_1, then_2)
   > SetIsEnabled(self=Get CreateButton() -> (CreateButton:Button Object Reference), bInIsEnabled=@n1)
   n3 = Get Message() -> (Message:String)
   n4 = Get MessageTextBlock() -> (MessageTextBlock:Text Object Reference)
   SetVisibility(execute=@n2.then_1, self=@n4, InVisibility=Select(Option 0=Visible, Option 1=Collapsed, Index=IsEmpty(InString=@n3)))
   > Set Text(Text="To Text (String)"(InString=Get Message()), self=@n4)
   n5 = Get Error() -> (Error:String)
   n6 = Get ErrorTextBlock() -> (ErrorTextBlock:Text Object Reference)
   SetVisibility(execute=@n2.then_2, self=@n6, InVisibility=Select(Option 0=Visible, Option 1=Collapsed, Index=IsEmpty(InString=@n5)))
   > Set Text(Text="To Text (String)"(InString=Get Error()), self=@n6)
  ]}
 Set Montage Name {nodes [
   n1 = Set Montage Name!() -> (then, Object:Object Reference) {is=title K2Node_FunctionEntry}
   > n2 = Cast To CreateAnimMontage_Object(Object=@n1)
   n3 = Get Name!(self:Create Anim Montage Object Object Reference=@n2) -> (Name:String)
   > Set Text(Text="To Text (String)"(InString=Append(A=AM_, B=@n3)), self:Text Box Object Reference=Get AnimMontageName())
  ]}
 Update AnimMontage Presets {nodes [
   Update AnimMontage Presets() {is=title K2Node_FunctionEntry}
   > Clear!(execute, TargetMap:Map of Integers to Strings=Get Montage Preset Map()) -> (then) {is=title K2Node_CallFunction}
   > n1 = GetDerivedClassNames!(
    execute
    self:Asset Registry Interface=GetAssetRegistry()
    ClassNames:Array of Top Level Asset Path Structures=Make Array("[0]":Top Level Asset Path Structure=GetSoftClassTopLevelAssetPath!(SoftClassReference:Object Soft Class Reference=/AnimInEdTools/AnimationBrowser/Dialog/CreateAnimMontage/MontagePreset/BlankMontage.BlankMontage_C) -> (ReturnValue:Top Level Asset Path Structure)) -> (Array:Array of Top Level Asset Path Structures)
    ExcludedClassNames:Set of Top Level Asset Path Structures=Make Set!() -> (Set:Set of Top Level Asset Path Structures) {is=title K2Node_MakeSet}
   ) -> (then, OutDerivedClassNames:Set of Top Level Asset Path Structures)
   > n2 = To Array!(execute, A:Set of Top Level Asset Path Structures=@n1) -> (then, Result:Array of Top Level Asset Path Structures)
   > n3 = For Each Loop(Array:Array of Top Level Asset Path Structures=@n2) -> (Array Element:Top Level Asset Path Structure)
   n4 = Get PresetComboBox()
   Branch(
    execute=@n3.Completed
    Condition="integer > integer"(A=GetOptionCount!(self:"Searchable Combo Box (String) Object Reference"=@n4) -> (ReturnValue:int=0))
   )
   > SetSelectedIndex!(execute, self:"Searchable Combo Box (String) Object Reference"=Get PresetComboBox(), Index:int=0) -> (then)
   n5 = BreakTopLevelAssetPath!(TopLevelAssetPath:"Top Level Asset Path Structure (by ref)"=@n3) -> (PathString:String)
   AddOption!(
    execute=@n3.LoopBody
    self:"Searchable Combo Box (String) Object Reference"=Get PresetComboBox()
    Option:String=GetBaseFilename(InPath=@n5, bRemovePath=true)
   ) -> (then)
   > Add!(execute, TargetMap:Map of Integers to Strings=Get Montage Preset Map(), Key:"Integer (by ref)"=@n3.Array Index, Value:"String (by ref)"=@n5) -> (then) {is=title K2Node_CallFunction}
  ]}
 Update AnimSequence ListView {nodes [
   Update AnimSequence ListView() {is=title K2Node_FunctionEntry}
   > ClearListItems(self=Get ListView())
   n1 = Get Level Sequence() -> (Level Sequence:Level Sequence Object Reference)
   > n2 = GetAnimSequenceLinkFromLevelSequence!(execute, InLevelSequence:Level Sequence Object Reference=@n1) -> (then, ReturnValue:Level Sequence Anim Sequence Link Object Reference)
   > n3 = For Each Loop(Array:Array of Level Sequence Anim Sequence Link Item Structures=Get AnimSequenceLinks!(self:Level Sequence Anim Sequence Link Object Reference=@n2) -> (AnimSequenceLinks:Array of Level Sequence Anim Sequence Link Item Structures)) -> (Array Element:Level Sequence Anim Sequence Link Item Structure)
   SetSelectedIndex(execute=@n3.Completed, self:List View Object Reference=Get ListView(), Index=0)
   n4 = Get Selected Item(self=Get ListView())
   > Is Valid(InputObject=@n4)
   > Set Montage Name(Object=@n4) {is=title K2Node_CallFunction}
   n5 = Break Level Sequence Anim Sequence Link Item!(LevelSequenceAnimSequenceLinkItem:"Level Sequence Anim Sequence Link Item Structure (by ref)"=@n3) -> (
    SkelTrackGuid:Guid Structure
    PathToAnimSequence:Soft Object Path Structure
    bAutoBake:bool=~
    bExportTransforms:bool=~
    bExportMorphTargets:bool=~
    bExportAttributeCurves:bool=~
    bExportMaterialCurves:bool=~
    bSkipCurvesWithZeroValue:bool=~
    bSetRetargetSourceAsset:bool=~
    Interpolation:EAnimInterpolationType Enum
    CurveInterpolation:ERichCurveInterpMode Enum
    bRecordInWorldSpace:bool=~
    bEvaluateAllSkeletalMeshComponents:bool=~
    IncludeAnimationNames:Array of Strings
    ExcludeAnimationNames:Array of Strings
    bRemoveExcludedCurves:bool=~
    WarmUpFrames:Frame Number Structure
    DelayBeforeStart:Frame Number Structure
    bUseCustomTimeRange:bool=~
    CustomStartFrame:Frame Number Structure
    CustomEndFrame:Frame Number Structure
    CustomDisplayRate:Frame Rate Structure
    bUseCustomFrameRate:bool=~
    CustomFrameRate:Frame Rate Structure
   ) {is=title K2Node_BreakStruct}
   n6 = FindBindingById!(execute=@n3.LoopBody, Sequence:Movie Scene Sequence Object Reference=@n1, BindingId:Guid Structure=@n5) -> (then, ReturnValue_BindingID:Guid Structure, ReturnValue_Sequence:Movie Scene Sequence Object Reference)
   n7 = Find(TargetMap=Get Asset Binding Map(), Key=GetSelectedIndex(self=Get AssetComboBox()) -> (ReturnValue=0))
   > Branch(Condition="Equal (Guid)"(
     A:"Guid Structure (by ref)"=@n6
     B:"Guid Structure (by ref)"=Break Movie Scene Binding Proxy!(MovieSceneBindingProxy:"Movie Scene Binding Proxy Structure (by ref)"=@n7) -> (BindingID:Guid Structure, Sequence:Movie Scene Sequence Object Reference) {is=title K2Node_BreakStruct}
    ) -> (ReturnValue:bool=~) {is=title K2Node_PromotableOperator})
   > n8 = Construct Object from Class(Class=/AnimInEdTools/AnimationBrowser/Dialog/CreateAnimMontage/Data/CreateAnimMontage_Object.CreateAnimMontage_Object_C) -> (ReturnValue:Create Anim Montage Object Object Reference)
   > n9 = Set Binding Object(Binding Object:Create Anim Montage Object Object Reference=@n8) -> (Output_Get:Create Anim Montage Object Object Reference)
   n10 = BreakSoftObjectPath(InSoftObjectPath=@n5.PathToAnimSequence)
   > Set Name!(execute, Name:String=GetBaseFilename(InPath=@n10, bRemovePath=true), self:Create Anim Montage Object Object Reference=@n9) -> (then, Output_Get:String)
   > Set AnimSequence Path!(
    execute
    AnimSequence Path:String=@n10
    self:Create Anim Montage Object Object Reference=Get Binding Object() -> (Binding Object:Create Anim Montage Object Object Reference)
   ) -> (then, Output_Get:String)
   > AddItem!(execute, self:List View Object Reference=Get ListView(), Item:Object Reference=Get Binding Object()) -> (then)
  ]}
 Update Bindings {nodes [
   Update Bindings!() -> (then, LevelSequence:Level Sequence Object Reference) {is=title K2Node_FunctionEntry}
   > n1 = Sequence()
   > ClearOptions(self:"Searchable Combo Box (String) Object Reference"=Get AssetComboBox())
   > Clear(TargetMap:Map of Integers to Movie Scene Binding Proxy Structures=Get Asset Binding Map())
   n2 = GetBindings!(execute=@n1.then_1, Sequence:Movie Scene Sequence Object Reference=Get LevelSequence() -> (LevelSequence:Level Sequence Object Reference)) -> (then, ReturnValue:Array of Movie Scene Binding Proxy Structures)
   > n3 = For Each Loop(Array:Array of Movie Scene Binding Proxy Structures=@n2) -> (Array Element:Movie Scene Binding Proxy Structure)
   > n4 = GetTracks!(execute, InBinding:"Movie Scene Binding Proxy Structure (by ref)"=@n3) -> (then, ReturnValue:Array of Movie Scene Track Object References)
   > n5 = For Each Loop with Break!(Exec:Exec, Array:Array of Movie Scene Track Object References=@n4, Break:Exec=@n10.then) -> (LoopBody:Exec, Array Element:Movie Scene Track Object Reference, Array Index:int, Completed:Exec)
   > n6 = Sequence!(execute) -> (then_0, then_1)
   > n7 = Cast To MovieSceneControlRigParameterTrack!(execute, Object=@n5) -> (then, CastFailed, AsMovie Scene Control Rig Parameter Track:Movie Scene Control Rig Parameter Track Object Reference) {is=title K2Node_DynamicCast}
   n8 = GetDisplayName!(execute:Exec=[@n7.then, @n11.then], InBinding:"Movie Scene Binding Proxy Structure (by ref)"=@n3) -> (then:Exec, ReturnValue:Text)
   > AddOption(self=Get AssetComboBox(), Option="To String (Text)"(InText=@n8))
   n9 = Get AssetComboBox()
   > n10 = Add(
    TargetMap:Map of Integers to Movie Scene Binding Proxy Structures=Get Asset Binding Map()
    Key=int - int(A=GetOptionCount(self=@n9) -> (ReturnValue=0))
    Value:"Movie Scene Binding Proxy Structure (by ref)"=@n3
   )
   n11 = Cast To MovieSceneSkeletalAnimationTrack!(execute=@n6.then_1, Object=@n5) -> (then, CastFailed, AsMovie Scene Skeletal Animation Track:Movie Scene Skeletal Animation Track Object Reference) {is=title K2Node_DynamicCast}
   SetSelectedIndex(execute=@n1.then_2, self:"Searchable Combo Box (String) Object Reference"=Get AssetComboBox(), Index=0)
  ]}
 Validate Data {nodes [Validate Data!() -> (
    then
    AnimMontage Name:String
    AnimSequence Path:String
    Montage Preset:Blank Montage Class Reference
    Binding Proxy:Movie Scene Binding Proxy Structure
    Widget:EUW Dialog Create Anim Montage Object Reference
   ) {is=title K2Node_FunctionEntry}]}
]
variables [
 Asset Binding Map {default {}}
 Cancel Button Pressed {default=()}
 Create Button Pressed {default=()}
 Level Sequence {type=LevelSequence, default=null}
 Montage Preset Map {default {}}
 Validate Data {default=()}
]
widgets {
 info {rootWidgetClass=&VerticalBox, widgetCount=31}
 widgets [
  VerticalBox_0 {widgetClassPath=&VerticalBox, widget=VerticalBox_0}
  SizeBox_1 {
   widgetClassPath=&SizeBox
   parent=VerticalBox_0
   widget=SizeBox_1
   properties {bOverride_MinDesiredHeight=true, minDesiredHeight=20}
   slot {class=VerticalBoxSlot, properties {verticalAlignment=VAlign_Center}}
  }
  Border_274 {
   widgetClassPath=&Border
   parent=SizeBox_1
   widget=Border_274
   properties {verticalAlignment=VAlign_Center, brushColor=<rgba 0.0275 0.0275 0.0275 1>, padding=<ltrb 5 5 5 5>}
   slot {class=SizeBoxSlot, properties {padding=<ltrb 0 0 0 2.5>}}
  }
  TitleTextBlock_1 {
   widgetClassPath=&TextBlock
   bIsVariable=true
   parent=Border_274
   widget=TitleTextBlock_1
   properties {
    text=Create Animation Montage
    colorAndOpacity {colorUseRule=UseColor_Specified, specifiedColor=<rgba 0.5255 0.5255 0.5255 1>}
    font {
     bForceMonospaced=false
     bMaterialIsStencil=false
     fontMaterial=null
     fontObject=&Roboto
     letterSpacing=0
     monospacedWidth=1
     size=10
     skewAmount=0
     typefaceFontName=Bold
     outlineSettings {
      bApplyOutlineToDropShadows=false
      bMiteredCorners=false
      bSeparateFillAlpha=false
      outlineMaterial=null
      outlineSize=0
      outlineColor=<rgba 0 0 0 1>
     }
    }
   }
   slot {class=BorderSlot, properties {verticalAlignment=VAlign_Center, padding=<ltrb 5 5 5 5>}}
  }
  Border_3 {
   widgetClassPath=&Border
   parent=VerticalBox_0
   widget=Border_3
   properties {brushColor=<rgba 0.0273 0.0273 0.0273 1>}
   slot {class=VerticalBoxSlot, properties {padding=<ltrb 0 2.5 0 2.5>, size {sizeRule=Fill, value=1}}}
  }
  NameVerticalBox {widgetClassPath=&VerticalBox, parent=Border_3, widget=NameVerticalBox, slot {class=BorderSlot}}
  AssetHorizontalBox {
   widgetClassPath=&VerticalBox
   parent=NameVerticalBox
   widget=AssetHorizontalBox
   slot {class=VerticalBoxSlot, properties {padding=<ltrb 0 1 0 1>}}
  }
  SizeBox_6 {
   widgetClassPath=&SizeBox
   parent=AssetHorizontalBox
   widget=SizeBox_6
   properties {bOverride_MinDesiredWidth=true, minDesiredWidth=100}
   slot {class=VerticalBoxSlot}
  }
  AssetTextBlock {
   widgetClassPath=&TextBlock
   bIsVariable=true
   parent=SizeBox_6
   widget=AssetTextBlock
   properties {
    text=Asset
    colorAndOpacity {colorUseRule=UseColor_Specified, specifiedColor=<rgba 0.5255 0.5255 0.5255 1>}
    font {
     bForceMonospaced=false
     bMaterialIsStencil=false
     fontMaterial=null
     fontObject=&Roboto
     letterSpacing=0
     monospacedWidth=1
     size=10
     skewAmount=0
     typefaceFontName=Bold
     outlineSettings {
      bApplyOutlineToDropShadows=false
      bMiteredCorners=false
      bSeparateFillAlpha=false
      outlineMaterial=null
      outlineSize=0
      outlineColor=<rgba 0 0 0 1>
     }
    }
   }
   slot {class=SizeBoxSlot, properties {horizontalAlignment=HAlign_Left}}
  }
  AssetComboBox {
   widgetClassPath=/Script/FortniteEditor.SearchableStringComboBox
   bIsVariable=true
   parent=AssetHorizontalBox
   widget=AssetComboBox
   properties {
    font {
     bForceMonospaced=false
     bMaterialIsStencil=false
     fontMaterial=null
     fontObject=&Roboto
     letterSpacing=0
     monospacedWidth=1
     size=10
     skewAmount=0
     typefaceFontName=Regular
     outlineSettings {
      bApplyOutlineToDropShadows=false
      bMiteredCorners=false
      bSeparateFillAlpha=false
      outlineMaterial=null
      outlineSize=0
      outlineColor=<rgba 0 0 0 1>
     }
    }
    foregroundColor {colorUseRule=UseColor_Specified, specifiedColor=<rgba 0.5271 0.5271 0.5271 1>}
    itemStyle {
     bUseParentRowBrush=false
     activeBrush {
      bIsDynamicallyLoaded=false
      drawAs=Image
      imageType=NoImage
      mirroring=NoMirror
      resourceName=None
      resourceObject=null
      tiling=NoTile
      imageSize=<0 0>
      margin=<ltrb 0 0 0 0>
      outlineSettings {
       bUseBrushTransparency=false
       roundingType=HalfHeightRadius
       width=0
       color {colorUseRule=UseColor_Specified, specifiedColor=<rgba 0 0 0 0>}
       cornerRadii=<0 0 0 0>
      }
      tintColor {colorUseRule=UseColor_Specified, specifiedColor=<rgba 0 0.162 0.7454 1>}
      uVRegion {bIsValid=false, max=<0 0>, min=<0 0>}
     }
     activeHighlightedBrush {
      bIsDynamicallyLoaded=false
      drawAs=Image
      imageType=NoImage
      mirroring=NoMirror
      resourceName=None
      resourceObject=null
      tiling=NoTile
      imageSize=<0 0>
      margin=<ltrb 0 0 0 0>
      outlineSettings {
       bUseBrushTransparency=false
       roundingType=HalfHeightRadius
       width=0
       color {colorUseRule=UseColor_Specified, specifiedColor=<rgba 0 0 0 0>}
       cornerRadii=<0 0 0 0>
      }
      tintColor {colorUseRule=UseColor_Specified, specifiedColor=<rgba 0.0252 0.0319 0.0423 1>}
      uVRegion {bIsValid=false, max=<0 0>, min=<0 0>}
     }
     activeHoveredBrush {
      bIsDynamicallyLoaded=false
      drawAs=Image
      imageType=NoImage
      mirroring=NoMirror
      resourceName=None
      resourceObject=null
      tiling=NoTile
      imageSize=<0 0>
      margin=<ltrb 0 0 0 0>
      outlineSettings {
       bUseBrushTransparency=false
       roundingType=HalfHeightRadius
       width=0
       color {colorUseRule=UseColor_Specified, specifiedColor=<rgba 0 0 0 0>}
       cornerRadii=<0 0 0 0>
      }
      tintColor {colorUseRule=UseColor_Specified, specifiedColor=<rgba 0 0.162 0.7454 1>}
      uVRegion {bIsValid=false, max=<0 0>, min=<0 0>}
     }
     dropIndicator_Above {
      bIsDynamicallyLoaded=false
      drawAs=Box
      imageType=FullColor
      mirroring=NoMirror
      resourceName=../../../Engine/Content/Slate/Common/DropZoneIndicator_Above.png
      resourceObject=null
      tiling=NoTile
      imageSize=<0 0>
      margin=<ltrb 0.625 0.625 0 0>
      outlineSettings {
       bUseBrushTransparency=false
       roundingType=HalfHeightRadius
       width=0
       color {colorUseRule=UseColor_Specified, specifiedColor=<rgba 0 0 0 0>}
       cornerRadii=<0 0 0 0>
      }
      tintColor {colorUseRule=UseColor_Specified, specifiedColor=<rgba 0.0091 0.3515 0.9647 1>}
      uVRegion {bIsValid=false, max=<0 0>, min=<0 0>}
     }
     dropIndicator_Below {
      bIsDynamicallyLoaded=false
      drawAs=Box
      imageType=FullColor
      mirroring=NoMirror
      resourceName=../../../Engine/Content/Slate/Common/DropZoneIndicator_Below.png
      resourceObject=null
      tiling=NoTile
      imageSize=<0 0>
      margin=<ltrb 0.625 0 0 0.625>
      outlineSettings {
       bUseBrushTransparency=false
       roundingType=HalfHeightRadius
       width=0
       color {colorUseRule=UseColor_Specified, specifiedColor=<rgba 0 0 0 0>}
       cornerRadii=<0 0 0 0>
      }
      tintColor {colorUseRule=UseColor_Specified, specifiedColor=<rgba 0.0091 0.3515 0.9647 1>}
      uVRegion {bIsValid=false, max=<0 0>, min=<0 0>}
     }
     dropIndicator_Onto {
      bIsDynamicallyLoaded=false
      drawAs=Box
      imageType=FullColor
      mirroring=NoMirror
      resourceName=../../../Engine/Content/Slate/Common/DropZoneIndicator_Onto.png
      resourceObject=null
      tiling=NoTile
      imageSize=<0 0>
      margin=<ltrb 0.25 0.25 0.25 0.25>
      outlineSettings {
       bUseBrushTransparency=false
       roundingType=HalfHeightRadius
       width=0
       color {colorUseRule=UseColor_Specified, specifiedColor=<rgba 0 0 0 0>}
       cornerRadii=<0 0 0 0>
      }
      tintColor {colorUseRule=UseColor_Specified, specifiedColor=<rgba 0.0091 0.3515 0.9647 1>}
      uVRegion {bIsValid=false, max=<0 0>, min=<0 0>}
     }
     evenRowBackgroundBrush {
      bIsDynamicallyLoaded=false
      drawAs=Image
      imageType=NoImage
      mirroring=NoMirror
      resourceName=None
      resourceObject=null
      tiling=NoTile
      imageSize=<0 0>
      margin=<ltrb 0 0 0 0>
      outlineSettings {
       bUseBrushTransparency=false
       roundingType=HalfHeightRadius
       width=0
       color {colorUseRule=UseColor_Specified, specifiedColor=<rgba 0 0 0 0>}
       cornerRadii=<0 0 0 0>
      }
      tintColor {colorUseRule=UseColor_Specified, specifiedColor=<rgba 0.0103 0.0103 0.0103 1>}
      uVRegion {bIsValid=false, max=<0 0>, min=<0 0>}
     }
     evenRowBackgroundHoveredBrush {
      bIsDynamicallyLoaded=false
      drawAs=Image
      imageType=NoImage
      mirroring=NoMirror
      resourceName=None
      resourceObject=null
      tiling=NoTile
      imageSize=<0 0>
      margin=<ltrb 0 0 0 0>
      outlineSettings {
       bUseBrushTransparency=false
       roundingType=HalfHeightRadius
       width=0
       color {colorUseRule=UseColor_Specified, specifiedColor=<rgba 0 0 0 0>}
       cornerRadii=<0 0 0 0>
      }
      tintColor {colorUseRule=UseColor_Specified, specifiedColor=<rgba 0.0176 0.0176 0.0176 1>}
      uVRegion {bIsValid=false, max=<0 0>, min=<0 0>}
     }
     inactiveBrush {
      bIsDynamicallyLoaded=false
      drawAs=Image
      imageType=NoImage
      mirroring=NoMirror
      resourceName=None
      resourceObject=null
      tiling=NoTile
      imageSize=<0 0>
      margin=<ltrb 0 0 0 0>
      outlineSettings {
       bUseBrushTransparency=false
       roundingType=HalfHeightRadius
       width=0
       color {colorUseRule=UseColor_Specified, specifiedColor=<rgba 0 0 0 0>}
       cornerRadii=<0 0 0 0>
      }
      tintColor {colorUseRule=UseColor_Specified, specifiedColor=<rgba 0.0513 0.0953 0.159 1>}
      uVRegion {bIsValid=false, max=<0 0>, min=<0 0>}
     }
     inactiveHighlightedBrush {
      bIsDynamicallyLoaded=false
      drawAs=Image
      imageType=NoImage
      mirroring=NoMirror
      resourceName=None
      resourceObject=null
      tiling=NoTile
      imageSize=<0 0>
      margin=<ltrb 0 0 0 0>
      outlineSettings {
       bUseBrushTransparency=false
       roundingType=HalfHeightRadius
       width=0
       color {colorUseRule=UseColor_Specified, specifiedColor=<rgba 0 0 0 0>}
       cornerRadii=<0 0 0 0>
      }
      tintColor {colorUseRule=UseColor_Specified, specifiedColor=<rgba 0.0252 0.0319 0.0423 1>}
      uVRegion {bIsValid=false, max=<0 0>, min=<0 0>}
     }
     inactiveHoveredBrush {
      bIsDynamicallyLoaded=false
      drawAs=Image
      imageType=NoImage
      mirroring=NoMirror
      resourceName=None
      resourceObject=null
      tiling=NoTile
      imageSize=<0 0>
      margin=<ltrb 0 0 0 0>
      outlineSettings {
       bUseBrushTransparency=false
       roundingType=HalfHeightRadius
       width=0
       color {colorUseRule=UseColor_Specified, specifiedColor=<rgba 0 0 0 0>}
       cornerRadii=<0 0 0 0>
      }
      tintColor {colorUseRule=UseColor_Specified, specifiedColor=<rgba 0.0513 0.0953 0.159 1>}
      uVRegion {bIsValid=false, max=<0 0>, min=<0 0>}
     }
     oddRowBackgroundBrush {
      bIsDynamicallyLoaded=false
      drawAs=Image
      imageType=NoImage
      mirroring=NoMirror
      resourceName=None
      resourceObject=null
      tiling=NoTile
      imageSize=<0 0>
      margin=<ltrb 0 0 0 0>
      outlineSettings {
       bUseBrushTransparency=false
       roundingType=HalfHeightRadius
       width=0
       color {colorUseRule=UseColor_Specified, specifiedColor=<rgba 0 0 0 0>}
       cornerRadii=<0 0 0 0>
      }
      tintColor {colorUseRule=UseColor_Specified, specifiedColor=<rgba 0.0103 0.0103 0.0103 1>}
      uVRegion {bIsValid=false, max=<0 0>, min=<0 0>}
     }
     oddRowBackgroundHoveredBrush {
      bIsDynamicallyLoaded=false
      drawAs=Image
      imageType=NoImage
      mirroring=NoMirror
      resourceName=None
      resourceObject=null
      tiling=NoTile
      imageSize=<0 0>
      margin=<ltrb 0 0 0 0>
      outlineSettings {
       bUseBrushTransparency=false
       roundingType=HalfHeightRadius
       width=0
       color {colorUseRule=UseColor_Specified, specifiedColor=<rgba 0 0 0 0>}
       cornerRadii=<0 0 0 0>
      }
      tintColor {colorUseRule=UseColor_Specified, specifiedColor=<rgba 0.0176 0.0176 0.0176 1>}
      uVRegion {bIsValid=false, max=<0 0>, min=<0 0>}
     }
     parentRowBackgroundBrush {
      bIsDynamicallyLoaded=false
      drawAs=Image
      imageType=NoImage
      mirroring=NoMirror
      resourceName=None
      resourceObject=null
      tiling=NoTile
      imageSize=<32 32>
      margin=<ltrb 0 0 0 0>
      outlineSettings {
       bUseBrushTransparency=false
       roundingType=HalfHeightRadius
       width=0
       color {colorUseRule=UseColor_Specified, specifiedColor=<rgba 0 0 0 0>}
       cornerRadii=<0 0 0 0>
      }
      tintColor {colorUseRule=UseColor_Specified, specifiedColor=<rgba 1 1 1 1>}
      uVRegion {bIsValid=false, max=<0 0>, min=<0 0>}
     }
     parentRowBackgroundHoveredBrush {
      bIsDynamicallyLoaded=false
      drawAs=Image
      imageType=NoImage
      mirroring=NoMirror
      resourceName=None
      resourceObject=null
      tiling=NoTile
      imageSize=<32 32>
      margin=<ltrb 0 0 0 0>
      outlineSettings {
       bUseBrushTransparency=false
       roundingType=HalfHeightRadius
       width=0
       color {colorUseRule=UseColor_Specified, specifiedColor=<rgba 0 0 0 0>}
       cornerRadii=<0 0 0 0>
      }
      tintColor {colorUseRule=UseColor_Specified, specifiedColor=<rgba 1 1 1 1>}
      uVRegion {bIsValid=false, max=<0 0>, min=<0 0>}
     }
     selectedTextColor {colorUseRule=UseColor_Specified, specifiedColor=<rgba 1 1 1 1>}
     selectorFocusedBrush {
      bIsDynamicallyLoaded=false
      drawAs=Border
      imageType=FullColor
      mirroring=NoMirror
      resourceName=../../../Engine/Content/Slate/Common/Selector.png
      resourceObject=null
      tiling=Both
      imageSize=<0 0>
      margin=<ltrb 0.25 0.25 0.25 0.25>
      outlineSettings {
       bUseBrushTransparency=false
       roundingType=HalfHeightRadius
       width=0
       color {colorUseRule=UseColor_Specified, specifiedColor=<rgba 0 0 0 0>}
       cornerRadii=<0 0 0 0>
      }
      tintColor {colorUseRule=UseColor_Specified, specifiedColor=<rgba 0.701 0.225 0.003 1>}
      uVRegion {bIsValid=false, max=<0 0>, min=<0 0>}
     }
     textColor {colorUseRule=UseColor_Specified, specifiedColor=<rgba 0.5271 0.5271 0.5271 1>}
    }
    widgetStyle {
     comboButtonStyle {
      downArrowAlign=VAlign_Center
      buttonStyle {
       clickedSlateSound {resourceObject=null}
       disabled {
        bIsDynamicallyLoaded=false
        drawAs=RoundedBox
        imageType=NoImage
        mirroring=NoMirror
        resourceName=None
        resourceObject=null
        tiling=NoTile
        imageSize=<0 0>
        margin=<ltrb 0 0 0 0>
        outlineSettings {
         bUseBrushTransparency=false
         roundingType=FixedRadius
         width=1
         color {colorUseRule=UseColor_ColorTable, specifiedColor=<rgba 0 0 0 0>}
         cornerRadii=<4 4 4 4>
        }
        tintColor {colorUseRule=UseColor_Specified, specifiedColor=<rgba 0.0395 0.0395 0.0395 1>}
        uVRegion {bIsValid=false, max=<0 0>, min=<0 0>}
       }
       disabledForeground {colorUseRule=UseColor_ColorTable, specifiedColor=<rgba 0 0 0 0>}
       hovered {
        bIsDynamicallyLoaded=false
        drawAs=RoundedBox
        imageType=NoImage
        mirroring=NoMirror
        resourceName=None
        resourceObject=null
        tiling=NoTile
        imageSize=<0 0>
        margin=<ltrb 0 0 0 0>
        outlineSettings {
         bUseBrushTransparency=false
         roundingType=FixedRadius
         width=1
         color {colorUseRule=UseColor_ColorTable, specifiedColor=<rgba 0 0 0 0>}
         cornerRadii=<4 4 4 4>
        }
        tintColor {colorUseRule=UseColor_Specified, specifiedColor=<rgba 0.0048 0.0048 0.0048 1>}
        uVRegion {bIsValid=false, max=<0 0>, min=<0 0>}
       }
       hoveredForeground {colorUseRule=UseColor_ColorTable, specifiedColor=<rgba 0 0 0 0>}
       hoveredSlateSound {resourceObject=null}
       normal {
        bIsDynamicallyLoaded=false
        drawAs=RoundedBox
        imageType=NoImage
        mirroring=NoMirror
        resourceName=None
        resourceObject=null
        tiling=NoTile
        imageSize=<0 0>
        margin=<ltrb 0 0 0 0>
        outlineSettings {
         bUseBrushTransparency=false
         roundingType=FixedRadius
         width=1
         color {colorUseRule=UseColor_ColorTable, specifiedColor=<rgba 0 0 0 0>}
         cornerRadii=<4 4 4 4>
        }
        tintColor {colorUseRule=UseColor_Specified, specifiedColor=<rgba 0.0048 0.0048 0.0048 1>}
        uVRegion {bIsValid=false, max=<0 0>, min=<0 0>}
       }
       normalForeground {colorUseRule=UseColor_ColorTable, specifiedColor=<rgba 0 0 0 0>}
       normalPadding=<ltrb 8 1 8 1>
       pressed {
        bIsDynamicallyLoaded=false
        drawAs=RoundedBox
        imageType=NoImage
        mirroring=NoMirror
        resourceName=None
        resourceObject=null
        tiling=NoTile
        imageSize=<0 0>
        margin=<ltrb 0 0 0 0>
        outlineSettings {
         bUseBrushTransparency=false
         roundingType=FixedRadius
         width=1
         color {colorUseRule=UseColor_ColorTable, specifiedColor=<rgba 0 0 0 0>}
         cornerRadii=<4 4 4 4>
        }
        tintColor {colorUseRule=UseColor_Specified, specifiedColor=<rgba 0.0048 0.0048 0.0048 1>}
        uVRegion {bIsValid=false, max=<0 0>, min=<0 0>}
       }
       pressedForeground {colorUseRule=UseColor_ColorTable, specifiedColor=<rgba 0 0 0 0>}
       pressedPadding=<ltrb 8 2 8 0>
       pressedSlateSound {resourceObject=null}
      }
      contentPadding=<ltrb 0 0 0 0>
      downArrowImage {
       bIsDynamicallyLoaded=false
       drawAs=Image
       imageType=Vector
       mirroring=NoMirror
       resourceName=../../../Engine/Content/Slate/Starship/CoreWidgets/ComboBox/wide-chevron-down.svg
       resourceObject=null
       tiling=NoTile
       imageSize=<16 16>
       margin=<ltrb 0 0 0 0>
       outlineSettings {
        bUseBrushTransparency=false
        roundingType=HalfHeightRadius
        width=0
        color {colorUseRule=UseColor_Specified, specifiedColor=<rgba 0 0 0 0>}
        cornerRadii=<0 0 0 0>
       }
       tintColor {colorUseRule=UseColor_Specified, specifiedColor=<rgba 1 1 1 1>}
       uVRegion {bIsValid=false, max=<0 0>, min=<0 0>}
      }
      downArrowPadding=<ltrb 2 2 2 2>
      menuBorderBrush {
       bIsDynamicallyLoaded=false
       drawAs=RoundedBox
       imageType=NoImage
       mirroring=NoMirror
       resourceName=None
       resourceObject=null
       tiling=NoTile
       imageSize=<0 0>
       margin=<ltrb 0 0 0 0>
       outlineSettings {
        bUseBrushTransparency=false
        roundingType=FixedRadius
        width=1
        color {colorUseRule=UseColor_ColorTable, specifiedColor=<rgba 0 0 0 0>}
        cornerRadii=<0 0 0 0>
       }
       tintColor {colorUseRule=UseColor_Specified, specifiedColor=<rgba 0.0395 0.0395 0.0395 1>}
       uVRegion {bIsValid=false, max=<0 0>, min=<0 0>}
      }
      menuBorderPadding=<ltrb 0 0 0 0>
      shadowColorAndOpacity=<rgba 0 0 0 1>
      shadowOffset=<0 0>
     }
     contentPadding=<ltrb 0 0 0 0>
     menuRowPadding=<ltrb 8 3 8 3>
     pressedSlateSound {resourceObject=null}
     selectionChangeSlateSound {resourceObject=null}
    }
   }
   slot {class=VerticalBoxSlot}
  }
  NameHorizontalBox {
   widgetClassPath=&VerticalBox
   parent=NameVerticalBox
   widget=NameHorizontalBox
   slot {class=VerticalBoxSlot, properties {padding=<ltrb 0 2 0 2>}}
  }
  SizeBox_4 {
   widgetClassPath=&SizeBox
   parent=NameHorizontalBox
   widget=SizeBox_4
   properties {bOverride_MinDesiredWidth=true, maxDesiredWidth=100, minDesiredWidth=100}
   slot {class=VerticalBoxSlot}
  }
  NameTextBlock {
   widgetClassPath=&TextBlock
   bIsVariable=true
   parent=SizeBox_4
   widget=NameTextBlock
   properties {
    text=AnimSequence
    toolTipText=What Am I doing?
    colorAndOpacity {colorUseRule=UseColor_Specified, specifiedColor=<rgba 0.5255 0.5255 0.5255 1>}
    font {
     bForceMonospaced=false
     bMaterialIsStencil=false
     fontMaterial=null
     fontObject=&Roboto
     letterSpacing=0
     monospacedWidth=1
     size=10
     skewAmount=0
     typefaceFontName=Bold
     outlineSettings {
      bApplyOutlineToDropShadows=false
      bMiteredCorners=false
      bSeparateFillAlpha=false
      outlineMaterial=null
      outlineSize=0
      outlineColor=<rgba 0 0 0 1>
     }
    }
   }
   slot {class=SizeBoxSlot, properties {horizontalAlignment=HAlign_Left, verticalAlignment=VAlign_Top}}
  }
  Border_130 {
   widgetClassPath=&Border
   parent=NameHorizontalBox
   widget=Border_130
   properties {brushColor=<rgba 0.0048 0.0048 0.0048 1>}
   slot {class=VerticalBoxSlot}
  }
  ListView {
   widgetClassPath=/Script/Blutility.EditorUtilityListView
   bIsVariable=true
   parent=Border_130
   widget=ListView
   properties {entryWidgetClass=/AnimInEdTools/AnimationBrowser/Dialog/CreateAnimMontage/Widget/CreateAnimMontage_ListItem.CreateAnimMontage_ListItem_C}
   slot {class=BorderSlot}
  }
  MontageNameHorizontalBox {
   widgetClassPath=&VerticalBox
   parent=NameVerticalBox
   widget=MontageNameHorizontalBox
   slot {class=VerticalBoxSlot, properties {padding=<ltrb 0 1 0 1>}}
  }
  SizeBox {
   widgetClassPath=&SizeBox
   parent=MontageNameHorizontalBox
   widget=SizeBox
   properties {bOverride_MinDesiredWidth=true, minDesiredWidth=100}
   slot {class=VerticalBoxSlot}
  }
  AssetTextBlock_1 {
   widgetClassPath=&TextBlock
   bIsVariable=true
   parent=SizeBox
   widget=AssetTextBlock_1
   properties {
    text=Preset
    colorAndOpacity {colorUseRule=UseColor_Specified, specifiedColor=<rgba 0.5255 0.5255 0.5255 1>}
    font {
     bForceMonospaced=false
     bMaterialIsStencil=false
     fontMaterial=null
     fontObject=&Roboto
     letterSpacing=0
     monospacedWidth=1
     size=10
     skewAmount=0
     typefaceFontName=Bold
     outlineSettings {
      bApplyOutlineToDropShadows=false
      bMiteredCorners=false
      bSeparateFillAlpha=false
      outlineMaterial=null
      outlineSize=0
      outlineColor=<rgba 0 0 0 1>
     }
    }
   }
   slot {class=SizeBoxSlot, properties {horizontalAlignment=HAlign_Left}}
  }
  PresetComboBox {
   widgetClassPath=&SearchableStringComboBox
   bIsVariable=true
   parent=MontageNameHorizontalBox
   widget=PresetComboBox
   properties {
    font {
     bForceMonospaced=false
     bMaterialIsStencil=false
     fontMaterial=null
     fontObject=&Roboto
     letterSpacing=0
     monospacedWidth=1
     size=10
     skewAmount=0
     typefaceFontName=Regular
     outlineSettings {
      bApplyOutlineToDropShadows=false
      bMiteredCorners=false
      bSeparateFillAlpha=false
      outlineMaterial=null
      outlineSize=0
      outlineColor=<rgba 0 0 0 1>
     }
    }
    foregroundColor {colorUseRule=UseColor_Specified, specifiedColor=<rgba 0.5271 0.5271 0.5271 1>}
    itemStyle {
     bUseParentRowBrush=false
     activeBrush {
      bIsDynamicallyLoaded=false
      drawAs=Image
      imageType=NoImage
      mirroring=NoMirror
      resourceName=None
      resourceObject=null
      tiling=NoTile
      imageSize=<0 0>
      margin=<ltrb 0 0 0 0>
      outlineSettings {
       bUseBrushTransparency=false
       roundingType=HalfHeightRadius
       width=0
       color {colorUseRule=UseColor_Specified, specifiedColor=<rgba 0 0 0 0>}
       cornerRadii=<0 0 0 0>
      }
      tintColor {colorUseRule=UseColor_Specified, specifiedColor=<rgba 0 0.162 0.7454 1>}
      uVRegion {bIsValid=false, max=<0 0>, min=<0 0>}
     }
     activeHighlightedBrush {
      bIsDynamicallyLoaded=false
      drawAs=Image
      imageType=NoImage
      mirroring=NoMirror
      resourceName=None
      resourceObject=null
      tiling=NoTile
      imageSize=<0 0>
      margin=<ltrb 0 0 0 0>
      outlineSettings {
       bUseBrushTransparency=false
       roundingType=HalfHeightRadius
       width=0
       color {colorUseRule=UseColor_Specified, specifiedColor=<rgba 0 0 0 0>}
       cornerRadii=<0 0 0 0>
      }
      tintColor {colorUseRule=UseColor_Specified, specifiedColor=<rgba 0.0252 0.0319 0.0423 1>}
      uVRegion {bIsValid=false, max=<0 0>, min=<0 0>}
     }
     activeHoveredBrush {
      bIsDynamicallyLoaded=false
      drawAs=Image
      imageType=NoImage
      mirroring=NoMirror
      resourceName=None
      resourceObject=null
      tiling=NoTile
      imageSize=<0 0>
      margin=<ltrb 0 0 0 0>
      outlineSettings {
       bUseBrushTransparency=false
       roundingType=HalfHeightRadius
       width=0
       color {colorUseRule=UseColor_Specified, specifiedColor=<rgba 0 0 0 0>}
       cornerRadii=<0 0 0 0>
      }
      tintColor {colorUseRule=UseColor_Specified, specifiedColor=<rgba 0 0.162 0.7454 1>}
      uVRegion {bIsValid=false, max=<0 0>, min=<0 0>}
     }
     dropIndicator_Above {
      bIsDynamicallyLoaded=false
      drawAs=Box
      imageType=FullColor
      mirroring=NoMirror
      resourceName=../../../Engine/Content/Slate/Common/DropZoneIndicator_Above.png
      resourceObject=null
      tiling=NoTile
      imageSize=<0 0>
      margin=<ltrb 0.625 0.625 0 0>
      outlineSettings {
       bUseBrushTransparency=false
       roundingType=HalfHeightRadius
       width=0
       color {colorUseRule=UseColor_Specified, specifiedColor=<rgba 0 0 0 0>}
       cornerRadii=<0 0 0 0>
      }
      tintColor {colorUseRule=UseColor_Specified, specifiedColor=<rgba 0.0091 0.3515 0.9647 1>}
      uVRegion {bIsValid=false, max=<0 0>, min=<0 0>}
     }
     dropIndicator_Below {
      bIsDynamicallyLoaded=false
      drawAs=Box
      imageType=FullColor
      mirroring=NoMirror
      resourceName=../../../Engine/Content/Slate/Common/DropZoneIndicator_Below.png
      resourceObject=null
      tiling=NoTile
      imageSize=<0 0>
      margin=<ltrb 0.625 0 0 0.625>
      outlineSettings {
       bUseBrushTransparency=false
       roundingType=HalfHeightRadius
       width=0
       color {colorUseRule=UseColor_Specified, specifiedColor=<rgba 0 0 0 0>}
       cornerRadii=<0 0 0 0>
      }
      tintColor {colorUseRule=UseColor_Specified, specifiedColor=<rgba 0.0091 0.3515 0.9647 1>}
      uVRegion {bIsValid=false, max=<0 0>, min=<0 0>}
     }
     dropIndicator_Onto {
      bIsDynamicallyLoaded=false
      drawAs=Box
      imageType=FullColor
      mirroring=NoMirror
      resourceName=../../../Engine/Content/Slate/Common/DropZoneIndicator_Onto.png
      resourceObject=null
      tiling=NoTile
      imageSize=<0 0>
      margin=<ltrb 0.25 0.25 0.25 0.25>
      outlineSettings {
       bUseBrushTransparency=false
       roundingType=HalfHeightRadius
       width=0
       color {colorUseRule=UseColor_Specified, specifiedColor=<rgba 0 0 0 0>}
       cornerRadii=<0 0 0 0>
      }
      tintColor {colorUseRule=UseColor_Specified, specifiedColor=<rgba 0.0091 0.3515 0.9647 1>}
      uVRegion {bIsValid=false, max=<0 0>, min=<0 0>}
     }
     evenRowBackgroundBrush {
      bIsDynamicallyLoaded=false
      drawAs=Image
      imageType=NoImage
      mirroring=NoMirror
      resourceName=None
      resourceObject=null
      tiling=NoTile
      imageSize=<0 0>
      margin=<ltrb 0 0 0 0>
      outlineSettings {
       bUseBrushTransparency=false
       roundingType=HalfHeightRadius
       width=0
       color {colorUseRule=UseColor_Specified, specifiedColor=<rgba 0 0 0 0>}
       cornerRadii=<0 0 0 0>
      }
      tintColor {colorUseRule=UseColor_Specified, specifiedColor=<rgba 0.0103 0.0103 0.0103 1>}
      uVRegion {bIsValid=false, max=<0 0>, min=<0 0>}
     }
     evenRowBackgroundHoveredBrush {
      bIsDynamicallyLoaded=false
      drawAs=Image
      imageType=NoImage
      mirroring=NoMirror
      resourceName=None
      resourceObject=null
      tiling=NoTile
      imageSize=<0 0>
      margin=<ltrb 0 0 0 0>
      outlineSettings {
       bUseBrushTransparency=false
       roundingType=HalfHeightRadius
       width=0
       color {colorUseRule=UseColor_Specified, specifiedColor=<rgba 0 0 0 0>}
       cornerRadii=<0 0 0 0>
      }
      tintColor {colorUseRule=UseColor_Specified, specifiedColor=<rgba 0.0176 0.0176 0.0176 1>}
      uVRegion {bIsValid=false, max=<0 0>, min=<0 0>}
     }
     inactiveBrush {
      bIsDynamicallyLoaded=false
      drawAs=Image
      imageType=NoImage
      mirroring=NoMirror
      resourceName=None
      resourceObject=null
      tiling=NoTile
      imageSize=<0 0>
      margin=<ltrb 0 0 0 0>
      outlineSettings {
       bUseBrushTransparency=false
       roundingType=HalfHeightRadius
       width=0
       color {colorUseRule=UseColor_Specified, specifiedColor=<rgba 0 0 0 0>}
       cornerRadii=<0 0 0 0>
      }
      tintColor {colorUseRule=UseColor_Specified, specifiedColor=<rgba 0.0513 0.0953 0.159 1>}
      uVRegion {bIsValid=false, max=<0 0>, min=<0 0>}
     }
     inactiveHighlightedBrush {
      bIsDynamicallyLoaded=false
      drawAs=Image
      imageType=NoImage
      mirroring=NoMirror
      resourceName=None
      resourceObject=null
      tiling=NoTile
      imageSize=<0 0>
      margin=<ltrb 0 0 0 0>
      outlineSettings {
       bUseBrushTransparency=false
       roundingType=HalfHeightRadius
       width=0
       color {colorUseRule=UseColor_Specified, specifiedColor=<rgba 0 0 0 0>}
       cornerRadii=<0 0 0 0>
      }
      tintColor {colorUseRule=UseColor_Specified, specifiedColor=<rgba 0.0252 0.0319 0.0423 1>}
      uVRegion {bIsValid=false, max=<0 0>, min=<0 0>}
     }
     inactiveHoveredBrush {
      bIsDynamicallyLoaded=false
      drawAs=Image
      imageType=NoImage
      mirroring=NoMirror
      resourceName=None
      resourceObject=null
      tiling=NoTile
      imageSize=<0 0>
      margin=<ltrb 0 0 0 0>
      outlineSettings {
       bUseBrushTransparency=false
       roundingType=HalfHeightRadius
       width=0
       color {colorUseRule=UseColor_Specified, specifiedColor=<rgba 0 0 0 0>}
       cornerRadii=<0 0 0 0>
      }
      tintColor {colorUseRule=UseColor_Specified, specifiedColor=<rgba 0.0513 0.0953 0.159 1>}
      uVRegion {bIsValid=false, max=<0 0>, min=<0 0>}
     }
     oddRowBackgroundBrush {
      bIsDynamicallyLoaded=false
      drawAs=Image
      imageType=NoImage
      mirroring=NoMirror
      resourceName=None
      resourceObject=null
      tiling=NoTile
      imageSize=<0 0>
      margin=<ltrb 0 0 0 0>
      outlineSettings {
       bUseBrushTransparency=false
       roundingType=HalfHeightRadius
       width=0
       color {colorUseRule=UseColor_Specified, specifiedColor=<rgba 0 0 0 0>}
       cornerRadii=<0 0 0 0>
      }
      tintColor {colorUseRule=UseColor_Specified, specifiedColor=<rgba 0.0103 0.0103 0.0103 1>}
      uVRegion {bIsValid=false, max=<0 0>, min=<0 0>}
     }
     oddRowBackgroundHoveredBrush {
      bIsDynamicallyLoaded=false
      drawAs=Image
      imageType=NoImage
      mirroring=NoMirror
      resourceName=None
      resourceObject=null
      tiling=NoTile
      imageSize=<0 0>
      margin=<ltrb 0 0 0 0>
      outlineSettings {
       bUseBrushTransparency=false
       roundingType=HalfHeightRadius
       width=0
       color {colorUseRule=UseColor_Specified, specifiedColor=<rgba 0 0 0 0>}
       cornerRadii=<0 0 0 0>
      }
      tintColor {colorUseRule=UseColor_Specified, specifiedColor=<rgba 0.0176 0.0176 0.0176 1>}
      uVRegion {bIsValid=false, max=<0 0>, min=<0 0>}
     }
     parentRowBackgroundBrush {
      bIsDynamicallyLoaded=false
      drawAs=Image
      imageType=NoImage
      mirroring=NoMirror
      resourceName=None
      resourceObject=null
      tiling=NoTile
      imageSize=<32 32>
      margin=<ltrb 0 0 0 0>
      outlineSettings {
       bUseBrushTransparency=false
       roundingType=HalfHeightRadius
       width=0
       color {colorUseRule=UseColor_Specified, specifiedColor=<rgba 0 0 0 0>}
       cornerRadii=<0 0 0 0>
      }
      tintColor {colorUseRule=UseColor_Specified, specifiedColor=<rgba 1 1 1 1>}
      uVRegion {bIsValid=false, max=<0 0>, min=<0 0>}
     }
     parentRowBackgroundHoveredBrush {
      bIsDynamicallyLoaded=false
      drawAs=Image
      imageType=NoImage
      mirroring=NoMirror
      resourceName=None
      resourceObject=null
      tiling=NoTile
      imageSize=<32 32>
      margin=<ltrb 0 0 0 0>
      outlineSettings {
       bUseBrushTransparency=false
       roundingType=HalfHeightRadius
       width=0
       color {colorUseRule=UseColor_Specified, specifiedColor=<rgba 0 0 0 0>}
       cornerRadii=<0 0 0 0>
      }
      tintColor {colorUseRule=UseColor_Specified, specifiedColor=<rgba 1 1 1 1>}
      uVRegion {bIsValid=false, max=<0 0>, min=<0 0>}
     }
     selectedTextColor {colorUseRule=UseColor_Specified, specifiedColor=<rgba 1 1 1 1>}
     selectorFocusedBrush {
      bIsDynamicallyLoaded=false
      drawAs=Border
      imageType=FullColor
      mirroring=NoMirror
      resourceName=../../../Engine/Content/Slate/Common/Selector.png
      resourceObject=null
      tiling=Both
      imageSize=<0 0>
      margin=<ltrb 0.25 0.25 0.25 0.25>
      outlineSettings {
       bUseBrushTransparency=false
       roundingType=HalfHeightRadius
       width=0
       color {colorUseRule=UseColor_Specified, specifiedColor=<rgba 0 0 0 0>}
       cornerRadii=<0 0 0 0>
      }
      tintColor {colorUseRule=UseColor_Specified, specifiedColor=<rgba 0.701 0.225 0.003 1>}
      uVRegion {bIsValid=false, max=<0 0>, min=<0 0>}
     }
     textColor {colorUseRule=UseColor_Specified, specifiedColor=<rgba 0.5271 0.5271 0.5271 1>}
    }
    widgetStyle {
     comboButtonStyle {
      downArrowAlign=VAlign_Center
      buttonStyle {
       clickedSlateSound {resourceObject=null}
       disabled {
        bIsDynamicallyLoaded=false
        drawAs=RoundedBox
        imageType=NoImage
        mirroring=NoMirror
        resourceName=None
        resourceObject=null
        tiling=NoTile
        imageSize=<0 0>
        margin=<ltrb 0 0 0 0>
        outlineSettings {
         bUseBrushTransparency=false
         roundingType=FixedRadius
         width=1
         color {colorUseRule=UseColor_ColorTable, specifiedColor=<rgba 0 0 0 0>}
         cornerRadii=<4 4 4 4>
        }
        tintColor {colorUseRule=UseColor_Specified, specifiedColor=<rgba 0.0395 0.0395 0.0395 1>}
        uVRegion {bIsValid=false, max=<0 0>, min=<0 0>}
       }
       disabledForeground {colorUseRule=UseColor_ColorTable, specifiedColor=<rgba 0 0 0 0>}
       hovered {
        bIsDynamicallyLoaded=false
        drawAs=RoundedBox
        imageType=NoImage
        mirroring=NoMirror
        resourceName=None
        resourceObject=null
        tiling=NoTile
        imageSize=<0 0>
        margin=<ltrb 0 0 0 0>
        outlineSettings {
         bUseBrushTransparency=false
         roundingType=FixedRadius
         width=1
         color {colorUseRule=UseColor_ColorTable, specifiedColor=<rgba 0 0 0 0>}
         cornerRadii=<4 4 4 4>
        }
        tintColor {colorUseRule=UseColor_Specified, specifiedColor=<rgba 0.0048 0.0048 0.0048 1>}
        uVRegion {bIsValid=false, max=<0 0>, min=<0 0>}
       }
       hoveredForeground {colorUseRule=UseColor_ColorTable, specifiedColor=<rgba 0 0 0 0>}
       hoveredSlateSound {resourceObject=null}
       normal {
        bIsDynamicallyLoaded=false
        drawAs=RoundedBox
        imageType=NoImage
        mirroring=NoMirror
        resourceName=None
        resourceObject=null
        tiling=NoTile
        imageSize=<0 0>
        margin=<ltrb 0 0 0 0>
        outlineSettings {
         bUseBrushTransparency=false
         roundingType=FixedRadius
         width=1
         color {colorUseRule=UseColor_ColorTable, specifiedColor=<rgba 0 0 0 0>}
         cornerRadii=<4 4 4 4>
        }
        tintColor {colorUseRule=UseColor_Specified, specifiedColor=<rgba 0.0048 0.0048 0.0048 1>}
        uVRegion {bIsValid=false, max=<0 0>, min=<0 0>}
       }
       normalForeground {colorUseRule=UseColor_ColorTable, specifiedColor=<rgba 0 0 0 0>}
       normalPadding=<ltrb 8 1 8 1>
       pressed {
        bIsDynamicallyLoaded=false
        drawAs=RoundedBox
        imageType=NoImage
        mirroring=NoMirror
        resourceName=None
        resourceObject=null
        tiling=NoTile
        imageSize=<0 0>
        margin=<ltrb 0 0 0 0>
        outlineSettings {
         bUseBrushTransparency=false
         roundingType=FixedRadius
         width=1
         color {colorUseRule=UseColor_ColorTable, specifiedColor=<rgba 0 0 0 0>}
         cornerRadii=<4 4 4 4>
        }
        tintColor {colorUseRule=UseColor_Specified, specifiedColor=<rgba 0.0048 0.0048 0.0048 1>}
        uVRegion {bIsValid=false, max=<0 0>, min=<0 0>}
       }
       pressedForeground {colorUseRule=UseColor_ColorTable, specifiedColor=<rgba 0 0 0 0>}
       pressedPadding=<ltrb 8 2 8 0>
       pressedSlateSound {resourceObject=null}
      }
      contentPadding=<ltrb 0 0 0 0>
      downArrowImage {
       bIsDynamicallyLoaded=false
       drawAs=Image
       imageType=Vector
       mirroring=NoMirror
       resourceName=../../../Engine/Content/Slate/Starship/CoreWidgets/ComboBox/wide-chevron-down.svg
       resourceObject=null
       tiling=NoTile
       imageSize=<16 16>
       margin=<ltrb 0 0 0 0>
       outlineSettings {
        bUseBrushTransparency=false
        roundingType=HalfHeightRadius
        width=0
        color {colorUseRule=UseColor_Specified, specifiedColor=<rgba 0 0 0 0>}
        cornerRadii=<0 0 0 0>
       }
       tintColor {colorUseRule=UseColor_Specified, specifiedColor=<rgba 1 1 1 1>}
       uVRegion {bIsValid=false, max=<0 0>, min=<0 0>}
      }
      downArrowPadding=<ltrb 2 2 2 2>
      menuBorderBrush {
       bIsDynamicallyLoaded=false
       drawAs=RoundedBox
       imageType=NoImage
       mirroring=NoMirror
       resourceName=None
       resourceObject=null
       tiling=NoTile
       imageSize=<0 0>
       margin=<ltrb 0 0 0 0>
       outlineSettings {
        bUseBrushTransparency=false
        roundingType=FixedRadius
        width=1
        color {colorUseRule=UseColor_ColorTable, specifiedColor=<rgba 0 0 0 0>}
        cornerRadii=<0 0 0 0>
       }
       tintColor {colorUseRule=UseColor_Specified, specifiedColor=<rgba 0.0395 0.0395 0.0395 1>}
       uVRegion {bIsValid=false, max=<0 0>, min=<0 0>}
      }
      menuBorderPadding=<ltrb 0 0 0 0>
      shadowColorAndOpacity=<rgba 0 0 0 1>
      shadowOffset=<0 0>
     }
     contentPadding=<ltrb 0 0 0 0>
     menuRowPadding=<ltrb 8 3 8 3>
     pressedSlateSound {resourceObject=null}
     selectionChangeSlateSound {resourceObject=null}
    }
   }
   slot {class=VerticalBoxSlot}
  }
  MontageNameHorizontalBox_1 {
   widgetClassPath=&VerticalBox
   parent=NameVerticalBox
   widget=MontageNameHorizontalBox_1
   slot {class=VerticalBoxSlot, properties {padding=<ltrb 0 1 0 1>}}
  }
  SizeBox_2 {
   widgetClassPath=&SizeBox
   parent=MontageNameHorizontalBox_1
   widget=SizeBox_2
   properties {bOverride_MinDesiredWidth=true, minDesiredWidth=100}
   slot {class=VerticalBoxSlot}
  }
  AssetTextBlock_2 {
   widgetClassPath=&TextBlock
   bIsVariable=true
   parent=SizeBox_2
   widget=AssetTextBlock_2
   properties {
    text=AnimMontage
    colorAndOpacity {colorUseRule=UseColor_Specified, specifiedColor=<rgba 0.5255 0.5255 0.5255 1>}
    font {
     bForceMonospaced=false
     bMaterialIsStencil=false
     fontMaterial=null
     fontObject=&Roboto
     letterSpacing=0
     monospacedWidth=1
     size=10
     skewAmount=0
     typefaceFontName=Bold
     outlineSettings {
      bApplyOutlineToDropShadows=false
      bMiteredCorners=false
      bSeparateFillAlpha=false
      outlineMaterial=null
      outlineSize=0
      outlineColor=<rgba 0 0 0 1>
     }
    }
   }
   slot {class=SizeBoxSlot, properties {horizontalAlignment=HAlign_Left}}
  }
  AnimMontageName {
   widgetClassPath=&EditorUtilityEditableTextBox
   bIsVariable=true
   parent=MontageNameHorizontalBox_1
   widget=AnimMontageName
   slot {class=VerticalBoxSlot}
  }
  Message_HorizontalBox {
   widgetClassPath=&HorizontalBox
   parent=VerticalBox_0
   widget=Message_HorizontalBox
   slot {class=VerticalBoxSlot, properties {verticalAlignment=VAlign_Bottom, padding=<ltrb 3 1 0 3>}}
  }
  MessageTextBlock {
   widgetClassPath=&TextBlock
   bIsVariable=true
   parent=Message_HorizontalBox
   widget=MessageTextBlock
   properties {
    justification=Right
    text=Text 1
    visibility=Collapsed
    colorAndOpacity {colorUseRule=UseColor_Specified, specifiedColor=<rgba 0.5255 0.5255 0.5255 1>}
    font {
     bForceMonospaced=false
     bMaterialIsStencil=false
     fontMaterial=null
     fontObject=&Roboto
     letterSpacing=0
     monospacedWidth=1
     size=9
     skewAmount=0
     typefaceFontName=Regular
     outlineSettings {
      bApplyOutlineToDropShadows=false
      bMiteredCorners=false
      bSeparateFillAlpha=false
      outlineMaterial=null
      outlineSize=0
      outlineColor=<rgba 0 0 0 1>
     }
    }
   }
   slot {class=HorizontalBoxSlot, properties {horizontalAlignment=HAlign_Left, verticalAlignment=VAlign_Center}}
  }
  ErrorTextBlock {
   widgetClassPath=&TextBlock
   bIsVariable=true
   parent=Message_HorizontalBox
   widget=ErrorTextBlock
   properties {
    text=Please Select an AnimSequence
    colorAndOpacity {colorUseRule=UseColor_Specified, specifiedColor=<rgba 1 0.0941 0.102 1>}
    font {
     bForceMonospaced=false
     bMaterialIsStencil=false
     fontMaterial=null
     fontObject=&Roboto
     letterSpacing=0
     monospacedWidth=1
     size=9
     skewAmount=0
     typefaceFontName=Bold
     outlineSettings {
      bApplyOutlineToDropShadows=false
      bMiteredCorners=false
      bSeparateFillAlpha=false
      outlineMaterial=null
      outlineSize=0
      outlineColor=<rgba 0 0 0 1>
     }
    }
   }
   slot {class=HorizontalBoxSlot, properties {horizontalAlignment=HAlign_Left, verticalAlignment=VAlign_Center, size {sizeRule=Fill, value=1}}}
  }
  ButtonBox {
   widgetClassPath=&HorizontalBox
   parent=VerticalBox_0
   widget=ButtonBox
   slot {class=VerticalBoxSlot, properties {verticalAlignment=VAlign_Bottom, padding=<ltrb 5 5 5 5>}}
  }
  CreateButton {
   widgetClassPath=&Button
   parent=ButtonBox
   widget=CreateButton
   properties {
    bIsEnabled=false
    widgetStyle {
     clickedSlateSound {resourceObject=null}
     disabled {
      bIsDynamicallyLoaded=false
      drawAs=RoundedBox
      imageType=NoImage
      mirroring=NoMirror
      resourceName=None
      resourceObject=null
      tiling=NoTile
      imageSize=<0 0>
      margin=<ltrb 0 0 0 0>
      outlineSettings {
       bUseBrushTransparency=false
       roundingType=FixedRadius
       width=1
       color {colorUseRule=UseColor_ColorTable, specifiedColor=<rgba 0 0 0 0>}
       cornerRadii=<4 4 4 4>
      }
      tintColor {colorUseRule=UseColor_Specified, specifiedColor=<rgba 0.0395 0.0395 0.0395 1>}
      uVRegion {bIsValid=false, max=<0 0>, min=<0 0>}
     }
     disabledForeground {colorUseRule=UseColor_ColorTable, specifiedColor=<rgba 0 0 0 0>}
     hovered {
      bIsDynamicallyLoaded=false
      drawAs=RoundedBox
      imageType=NoImage
      mirroring=NoMirror
      resourceName=None
      resourceObject=null
      tiling=NoTile
      imageSize=<32 32>
      margin=<ltrb 0 0 0 0>
      outlineSettings {
       bUseBrushTransparency=false
       roundingType=FixedRadius
       width=1
       color {colorUseRule=UseColor_ColorTable, specifiedColor=<rgba 0 0 0 0>}
       cornerRadii=<4 4 4 4>
      }
      tintColor {colorUseRule=UseColor_Specified, specifiedColor=<rgba 0.0953 0.0953 0.0953 1>}
      uVRegion {bIsValid=false, max=<0 0>, min=<0 0>}
     }
     hoveredForeground {colorUseRule=UseColor_ColorTable, specifiedColor=<rgba 0 0 0 0>}
     hoveredSlateSound {resourceObject=null}
     normal {
      bIsDynamicallyLoaded=false
      drawAs=RoundedBox
      imageType=NoImage
      mirroring=NoMirror
      resourceName=None
      resourceObject=null
      tiling=NoTile
      imageSize=<32 32>
      margin=<ltrb 0 0 0 0>
      outlineSettings {
       bUseBrushTransparency=false
       roundingType=FixedRadius
       width=1
       color {colorUseRule=UseColor_ColorTable, specifiedColor=<rgba 0 0 0 0>}
       cornerRadii=<4 4 4 4>
      }
      tintColor {colorUseRule=UseColor_Specified, specifiedColor=<rgba 0.0395 0.0395 0.0395 1>}
      uVRegion {bIsValid=false, max=<0 0>, min=<0 0>}
     }
     normalForeground {colorUseRule=UseColor_ColorTable, specifiedColor=<rgba 0 0 0 0>}
     normalPadding=<ltrb 12 1.5 12 1.5>
     pressed {
      bIsDynamicallyLoaded=false
      drawAs=RoundedBox
      imageType=NoImage
      mirroring=NoMirror
      resourceName=None
      resourceObject=null
      tiling=NoTile
      imageSize=<32 32>
      margin=<ltrb 0 0 0 0>
      outlineSettings {
       bUseBrushTransparency=false
       roundingType=FixedRadius
       width=1
       color {colorUseRule=UseColor_ColorTable, specifiedColor=<rgba 0 0 0 0>}
       cornerRadii=<4 4 4 4>
      }
      tintColor {colorUseRule=UseColor_Specified, specifiedColor=<rgba 0.0284 0.0284 0.0284 1>}
      uVRegion {bIsValid=false, max=<0 0>, min=<0 0>}
     }
     pressedForeground {colorUseRule=UseColor_ColorTable, specifiedColor=<rgba 0 0 0 0>}
     pressedPadding=<ltrb 12 2.5 12 0.5>
     pressedSlateSound {resourceObject=null}
    }
   }
   slot {class=HorizontalBoxSlot, properties {size {sizeRule=Fill, value=1}}}
  }
  CreateButtonTextBlock {
   widgetClassPath=&TextBlock
   bIsVariable=true
   parent=CreateButton
   widget=CreateButtonTextBlock
   properties {
    autoWrapText=true
    justification=Center
    text=Create
    colorAndOpacity {colorUseRule=UseColor_Specified, specifiedColor=<rgba 0.5255 0.5255 0.5255 1>}
    font {
     bForceMonospaced=false
     bMaterialIsStencil=false
     fontMaterial=null
     fontObject=&Roboto
     letterSpacing=0
     monospacedWidth=1
     size=10
     skewAmount=0
     typefaceFontName=Bold
     outlineSettings {
      bApplyOutlineToDropShadows=false
      bMiteredCorners=false
      bSeparateFillAlpha=false
      outlineMaterial=null
      outlineSize=0
      outlineColor=<rgba 0 0 0 1>
     }
    }
   }
   slot {class=ButtonSlot}
  }
  CancelButton {
   widgetClassPath=&Button
   parent=ButtonBox
   widget=CancelButton
   properties {widgetStyle {
     clickedSlateSound {resourceObject=null}
     disabled {
      bIsDynamicallyLoaded=false
      drawAs=RoundedBox
      imageType=NoImage
      mirroring=NoMirror
      resourceName=None
      resourceObject=null
      tiling=NoTile
      imageSize=<0 0>
      margin=<ltrb 0 0 0 0>
      outlineSettings {
       bUseBrushTransparency=false
       roundingType=FixedRadius
       width=1
       color {colorUseRule=UseColor_ColorTable, specifiedColor=<rgba 0 0 0 0>}
       cornerRadii=<4 4 4 4>
      }
      tintColor {colorUseRule=UseColor_Specified, specifiedColor=<rgba 0.0395 0.0395 0.0395 1>}
      uVRegion {bIsValid=false, max=<0 0>, min=<0 0>}
     }
     disabledForeground {colorUseRule=UseColor_ColorTable, specifiedColor=<rgba 0 0 0 0>}
     hovered {
      bIsDynamicallyLoaded=false
      drawAs=RoundedBox
      imageType=NoImage
      mirroring=NoMirror
      resourceName=None
      resourceObject=null
      tiling=NoTile
      imageSize=<32 32>
      margin=<ltrb 0 0 0 0>
      outlineSettings {
       bUseBrushTransparency=false
       roundingType=FixedRadius
       width=1
       color {colorUseRule=UseColor_ColorTable, specifiedColor=<rgba 0 0 0 0>}
       cornerRadii=<4 4 4 4>
      }
      tintColor {colorUseRule=UseColor_Specified, specifiedColor=<rgba 0.0953 0.0953 0.0953 1>}
      uVRegion {bIsValid=false, max=<0 0>, min=<0 0>}
     }
     hoveredForeground {colorUseRule=UseColor_ColorTable, specifiedColor=<rgba 0 0 0 0>}
     hoveredSlateSound {resourceObject=null}
     normal {
      bIsDynamicallyLoaded=false
      drawAs=RoundedBox
      imageType=NoImage
      mirroring=NoMirror
      resourceName=None
      resourceObject=null
      tiling=NoTile
      imageSize=<32 32>
      margin=<ltrb 0 0 0 0>
      outlineSettings {
       bUseBrushTransparency=false
       roundingType=FixedRadius
       width=1
       color {colorUseRule=UseColor_ColorTable, specifiedColor=<rgba 0 0 0 0>}
       cornerRadii=<4 4 4 4>
      }
      tintColor {colorUseRule=UseColor_Specified, specifiedColor=<rgba 0.0395 0.0395 0.0395 1>}
      uVRegion {bIsValid=false, max=<0 0>, min=<0 0>}
     }
     normalForeground {colorUseRule=UseColor_ColorTable, specifiedColor=<rgba 0 0 0 0>}
     normalPadding=<ltrb 12 1.5 12 1.5>
     pressed {
      bIsDynamicallyLoaded=false
      drawAs=RoundedBox
      imageType=NoImage
      mirroring=NoMirror
      resourceName=None
      resourceObject=null
      tiling=NoTile
      imageSize=<32 32>
      margin=<ltrb 0 0 0 0>
      outlineSettings {
       bUseBrushTransparency=false
       roundingType=FixedRadius
       width=1
       color {colorUseRule=UseColor_ColorTable, specifiedColor=<rgba 0 0 0 0>}
       cornerRadii=<4 4 4 4>
      }
      tintColor {colorUseRule=UseColor_Specified, specifiedColor=<rgba 0.0284 0.0284 0.0284 1>}
      uVRegion {bIsValid=false, max=<0 0>, min=<0 0>}
     }
     pressedForeground {colorUseRule=UseColor_ColorTable, specifiedColor=<rgba 0 0 0 0>}
     pressedPadding=<ltrb 12 2.5 12 0.5>
     pressedSlateSound {resourceObject=null}
    }}
   slot {class=HorizontalBoxSlot, properties {size {sizeRule=Fill, value=1}}}
  }
  CancelTextBlock {
   widgetClassPath=&TextBlock
   bIsVariable=true
   parent=CancelButton
   widget=CancelTextBlock
   properties {
    autoWrapText=true
    justification=Center
    text=Cancel
    colorAndOpacity {colorUseRule=UseColor_Specified, specifiedColor=<rgba 0.5255 0.5255 0.5255 1>}
    font {
     bForceMonospaced=false
     bMaterialIsStencil=false
     fontMaterial=null
     fontObject=&Roboto
     letterSpacing=0
     monospacedWidth=1
     size=10
     skewAmount=0
     typefaceFontName=Bold
     outlineSettings {
      bApplyOutlineToDropShadows=false
      bMiteredCorners=false
      bSeparateFillAlpha=false
      outlineMaterial=null
      outlineSize=0
      outlineColor=<rgba 0 0 0 1>
     }
    }
   }
   slot {class=ButtonSlot}
  }
 ]
}
