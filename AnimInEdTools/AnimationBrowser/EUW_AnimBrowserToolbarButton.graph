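# Editor utility widget for a configurable icon button in the Animation Browser toolbar.
# Update stores the owning project options object, context and function name, sets tooltip and icon from an options struct, and binds Clicked to that owner.
# A click broadcasts Clicked, then, if CallFunctionByNameOnClick is set, waits one tick and calls the named function on the owner.
# That call uses inline Python `call_method`, so the target function is chosen by string.
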EditorUtilityWidgetBlueprint /AnimInEdTools/AnimationBrowser/EUW_AnimBrowserToolbarButton
graphs [
 Clicked {nodes [Clicked!() -> (then, Button:EUW Anim Browser Toolbar Button Object Reference) {is=title K2Node_FunctionEntry}]}
 EventGraph {nodes [
   "On Clicked (ToolbarButton)"() {is=title K2Node_ComponentBoundEvent}
   > Call Clicked!(
    execute:Exec
    self:EUW Anim Browser Toolbar Button Object Reference
    Button:EUW Anim Browser Toolbar Button Object Reference=Self-Reference()
   ) -> (then:Exec) {is=title K2Node_CallDelegate}
   > n1 = Branch(Condition=Get Call Function By Name OnClick() -> (Call Function By Name OnClick:bool))
   n2 = CustomEvent() {is=title K2Node_CustomEvent}
   Set Timer for Next Tick by Event!(execute=@n1.then, Delegate:Delegate=@n2) -> (then, ReturnValue:Timer Handle Structure) {is=title K2Node_CallFunction}
   Event PreConstruct()
   Event Construct()
   Event Tick()
   Execute Python Script!(
    execute=@n2.then
    PythonScript="import unreal\r\nowner.call_method(function_name)"
    function_name:String=Get Function To Call() -> (Function To Call:String)
    owner:EU Animation Browser Project Options Object Reference=Get OwningOption() -> (OwningOption:EU Animation Browser Project Options Object Reference)
   ) -> (then, ReturnValue) {inputs [function_name, owner]}
   Get Call Function By Name OnClick()
  ]}
 Update {nodes [
   n1 = Update!() -> (then, Option:S AB Toolbar Options Structure, ProjectOptions:EU Animation Browser Project Options Object Reference) {is=title K2Node_FunctionEntry}
   > Set OwningOption(OwningOption:EU Animation Browser Project Options Object Reference=@n1.ProjectOptions) -> (Output_Get:EU Animation Browser Project Options Object Reference)
   n2 = Break S AB Toolbar Options!(S_AB_ToolbarOptions:"S AB Toolbar Options Structure (by ref)"=@n1) -> (
    Enabled_2_#1:bool=~
    Context_13_#2:Name
    Tooltip_23_#3:String
    FunctionToCall_10_#4:String
    ButtonIcon_22_#5:Texture 2D Object Reference
   ) {is=title K2Node_BreakStruct}
   > Set Context(Context:Name=@n2.Context_13_#2) -> (Output_Get:Name=None)
   > SetToolTipText!(
    execute
    self:Widget Object Reference
    InToolTipText:"Text (by ref)"="To Text (String)"(InString=@n2.Tooltip_23_#3)
   ) -> (then)
   > Set Function To Call(Function To Call:String=@n2.FunctionToCall_10_#4) -> (Output_Get:String)
   > Set members in Slate Brush!(
    execute:Exec
    StructRef:"Slate Brush Structure (by ref)"=Get Brush!(self:Image Object Reference=Get Image() -> (Image:Image Object Reference)) -> (Brush:Slate Brush Structure)
    ResourceObject:Object Reference=@n2.ButtonIcon_22_#5
   ) -> (then:Exec, StructOut:"Slate Brush Structure (by ref)") {is=title K2Node_SetFieldsInStruct}
   > Bind Event to Clicked(self:EUW Anim Browser Toolbar Button Object Reference, Delegate=Create Event(self=Get OwningOption())) {is=title K2Node_AddDelegate}
  ]}
]
variables [
 Call Function By Name OnClick {type=boolean, default=true}
 Clicked {default=()}
 Context {type=string, default=Temp}
 Function To Call {type=string, default=""}
 Image Brush {
  type=SlateBrush
  default {
   bIsDynamicallyLoaded=false
   drawAs=Image
   imageType=FullColor
   mirroring=NoMirror
   resourceName=None
   resourceObject=null
   tiling=NoTile
   imageSize=<14 14>
   margin=<ltrb 0 0 0 0>
   outlineSettings {
    bUseBrushTransparency=false
    roundingType=HalfHeightRadius
    width=0
    color {colorUseRule=UseColor_Specified, specifiedColor=<rgba 0 0 0 0>}
    cornerRadii=<0 0 0 0>
   }
   tintColor {colorUseRule=UseColor_Specified, specifiedColor=<rgba 0.5271 0.5271 0.5271 1>}
   uVRegion {bIsValid=false, max=<0 0>, min=<0 0>}
  }
 }
 OwningOption {type=EU_AnimationBrowser_ProjectOptions_C, default=null}
]
widgets {
 info {rootWidgetClass=&EditorUtilityButton, widgetCount=2}
 widgets [
  ToolbarButton {
   widgetClassPath=&EditorUtilityButton
   widget=ToolbarButton
   properties {widgetStyle {
     clickedSlateSound {resourceObject=null}
     disabled {
      bIsDynamicallyLoaded=false
      drawAs=RoundedBox
      imageType=NoImage
      mirroring=NoMirror
      resourceName=None
      resourceObject=null
      tiling=NoTile
      imageSize=<16 16>
      margin=<ltrb 0 0 0 0>
      outlineSettings {
       bUseBrushTransparency=false
       roundingType=FixedRadius
       width=1
       color {colorUseRule=UseColor_ColorTable, specifiedColor=<rgba 0 0 0 0>}
       cornerRadii=<4 4 4 4>
      }
      tintColor {colorUseRule=UseColor_Specified, specifiedColor=<rgba 0.6146 0.6146 0.6146 1>}
      uVRegion {bIsValid=false, max=<0 0>, min=<0 0>}
     }
     disabledForeground {colorUseRule=UseColor_ColorTable, specifiedColor=<rgba 0 0 0 0>}
     hovered {
      bIsDynamicallyLoaded=false
      drawAs=RoundedBox
      imageType=NoImage
      mirroring=NoMirror
      resourceName=None
      resourceObject=null
      tiling=NoTile
      imageSize=<16 16>
      margin=<ltrb 0 0 0 0>
      outlineSettings {
       bUseBrushTransparency=false
       roundingType=FixedRadius
       width=0
       color {colorUseRule=UseColor_ColorTable, specifiedColor=<rgba 0 0 0 0>}
       cornerRadii=<4 4 4 4>
      }
      tintColor {colorUseRule=UseColor_Specified, specifiedColor=<rgba 0.0941 0.0941 0.0941 1>}
      uVRegion {bIsValid=false, max=<0 0>, min=<0 0>}
     }
     hoveredForeground {colorUseRule=UseColor_ColorTable, specifiedColor=<rgba 0 0 0 0>}
     hoveredSlateSound {resourceObject=null}
     normal {
      bIsDynamicallyLoaded=false
      drawAs=RoundedBox
      imageType=NoImage
      mirroring=NoMirror
      resourceName=None
      resourceObject=null
      tiling=NoTile
      imageSize=<16 16>
      margin=<ltrb 0 0 0 0>
      outlineSettings {
       bUseBrushTransparency=false
       roundingType=FixedRadius
       width=0
       color {colorUseRule=UseColor_ColorTable, specifiedColor=<rgba 0 0 0 0>}
       cornerRadii=<4 4 4 4>
      }
      tintColor {colorUseRule=UseColor_Specified, specifiedColor=<rgba 0.0208 0.0208 0.0208 0>}
      uVRegion {bIsValid=false, max=<0 0>, min=<0 0>}
     }
     normalForeground {colorUseRule=UseColor_ColorTable, specifiedColor=<rgba 0 0 0 0>}
     normalPadding=<ltrb 0 0 0 0>
     pressed {
      bIsDynamicallyLoaded=false
      drawAs=RoundedBox
      imageType=NoImage
      mirroring=NoMirror
      resourceName=None
      resourceObject=null
      tiling=NoTile
      imageSize=<16 16>
      margin=<ltrb 0 0 0 0>
      outlineSettings {
       bUseBrushTransparency=false
       roundingType=FixedRadius
       width=0
       color {colorUseRule=UseColor_ColorTable, specifiedColor=<rgba 0 0 0 0>}
       cornerRadii=<4 4 4 4>
      }
      tintColor {colorUseRule=UseColor_Specified, specifiedColor=<rgba 0.0275 0.0275 0.0275 1>}
      uVRegion {bIsValid=false, max=<0 0>, min=<0 0>}
     }
     pressedForeground {colorUseRule=UseColor_ColorTable, specifiedColor=<rgba 0 0 0 0>}
     pressedPadding=<ltrb 0 0 0 0>
     pressedSlateSound {resourceObject=null}
    }}
   slot=~
  }
  Image {
   widgetClassPath=&Image
   bIsVariable=true
   parent=ToolbarButton
   widget=Image
   properties {brush {
     bIsDynamicallyLoaded=false
     drawAs=Image
     imageType=FullColor
     mirroring=NoMirror
     resourceName=None
     resourceObject=null
     tiling=NoTile
     imageSize=<14 14>
     margin=<ltrb 0 0 0 0>
     outlineSettings {
      bUseBrushTransparency=false
      roundingType=HalfHeightRadius
      width=0
      color {colorUseRule=UseColor_Specified, specifiedColor=<rgba 0 0 0 0>}
      cornerRadii=<0 0 0 0>
     }
     tintColor {colorUseRule=UseColor_Specified, specifiedColor=<rgba 0.5271 0.5271 0.5271 1>}
     uVRegion {bIsValid=false, max=<0 0>, min=<0 0>}
    }}
   slot {class=ButtonSlot, properties {padding=<ltrb 3 3 3 3>}}
  }
 ]
}
