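# Editor utility widget that shows a bordered info panel for the animation browser dialog, with a title, two label/value rows and two error lines.
# Set Display Info fills all fields, while Set Text 1/2 and Set Error 1/2 update individual parts, with Error 2 taking a red-by-default color.
# Each path calls Set Display Visibility, which collapses any line with empty text, hiding a row only when both label and value are empty.
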EditorUtilityWidgetBlueprint /AnimInEdTools/AnimationBrowser/EUW_AnimationDialog_BorderInfo
graphs [
 EventGraph {nodes [
   n1 = Set Display Info!() -> (OutputDelegate, then, Title:Text, Title 1:Text, Text 1:Text, Title 2:Text, Text 2:Text, Error 1:Text, Error 2:Text) {is=title K2Node_CustomEvent}
   > Set Text(Text=@n1.Title, self=Get TitleTextBlock() -> (TitleTextBlock:Text Object Reference))
   > Set Text(Text=@n1.Title 1, self=Get PathTitle1TextBlock() -> (PathTitle1TextBlock:Text Object Reference))
   > Set Text(Text=@n1.Text 1, self=Get Path1TextBlock() -> (Path1TextBlock:Text Object Reference))
   > Set Text(Text=@n1.Title 2, self=Get Title2TextBlock() -> (Title2TextBlock:Text Object Reference))
   > Set Text(Text=@n1.Text 2, self=Get Text2TextBlock() -> (Text2TextBlock:Text Object Reference))
   > Set Text(Text=@n1.Error 1, self=Get Error1TextBlock() -> (Error1TextBlock:Text Object Reference))
   > Set Text(Text=@n1.Error 2, self=Get Error2TextBlock() -> (Error2TextBlock:Text Object Reference))
   > Set Display Visibility() {is=title K2Node_CallFunction}
   > SetVisibility(InVisibility=Visible)
   Set Display Visibility() {is=title K2Node_CustomEvent}
   n2 = TextIsEmpty(InText="GetText (Text)"!(self:Text Object Reference=Get Text2TextBlock()) -> (ReturnValue:Text))
   n3 = TextIsEmpty(InText="GetText (Text)"(self=Get Title2TextBlock()))
   > SetVisibility(
    self=Get Text2HorizontalBox() -> (Text2HorizontalBox:Horizontal Box Object Reference)
    InVisibility=Select(Option 0=Visible, Option 1=Collapsed, Index=AND Boolean(A=@n2, B=@n3))
   )
   n4 = TextIsEmpty(InText="GetText (Text)"(self=Get Path1TextBlock()))
   n5 = TextIsEmpty(InText="GetText (Text)"(self=Get PathTitle1TextBlock()))
   > SetVisibility(
    self=Get Text1HorizontalBox() -> (Text1HorizontalBox:Horizontal Box Object Reference)
    InVisibility=Select(Option 0=Visible, Option 1=Collapsed, Index=AND Boolean(A=@n4, B=@n5))
   )
   n6 = "GetText (Text)"(self=Get Error1TextBlock())
   > SetVisibility(self=Get Error1TextBlock(), InVisibility=Select(Option 0=Visible, Option 1=Collapsed, Index=TextIsEmpty(InText=@n6)))
   n7 = "GetText (Text)"(self=Get TitleTextBlock())
   > SetVisibility(self=Get TitleTextBlock(), InVisibility=Select(Option 0=Visible, Option 1=Collapsed, Index=TextIsEmpty(InText=@n7)))
   n8 = "GetText (Text)"(self=Get Error2TextBlock())
   > SetVisibility(self=Get Error2TextBlock(), InVisibility=Select(Option 0=Visible, Option 1=Collapsed, Index=TextIsEmpty(InText=@n8)))
   Event PreConstruct()
   Event Construct()
   Event Tick()
  ]}
 Set Error 1 {nodes [
   n1 = Set Error 1!() -> (then, Text:Text) {is=title K2Node_FunctionEntry}
   > Set Text(Text=@n1, self=Get Error1TextBlock())
   > Set Display Visibility() {is=title K2Node_CallFunction}
  ]}
 Set Error 2 {nodes [
   n1 = Set Error 2!() -> (then, Text:Text, Specified Color:Linear Color Structure=(R=1,G=0.1,B=0.1,A=1)) {is=title K2Node_FunctionEntry}
   n2 = Get Error2TextBlock()
   > Set Text(Text=@n1, self=@n2)
   > Set Display Visibility()
   > Set ColorAndOpacity!(execute, ColorAndOpacity:Slate Color Structure=Make Slate Color(SpecifiedColor=@n1.Specified Color), self:Text Object Reference=@n2) -> (then, Output_Get:Slate Color Structure)
  ]}
 Set Text 1 {nodes [
   n1 = Set Text 1!() -> (then, Title:Text, Text:Text) {is=title K2Node_FunctionEntry}
   > Set Text(Text=@n1, self=Get PathTitle1TextBlock())
   > Set Text(Text=@n1.Text, self=Get Path1TextBlock())
   > Set Display Visibility()
  ]}
 Set Text 2 {nodes [
   n1 = Set Text 2!() -> (then, Title:Text, Text:Text) {is=title K2Node_FunctionEntry}
   > Set Text(Text=@n1, self=Get Title2TextBlock())
   > Set Text(Text=@n1.Text, self=Get Text2TextBlock())
   > Set Display Visibility()
  ]}
]
variables []
widgets {
 info {rootWidgetClass=&Border, widgetCount=11}
 widgets [
  Border {widgetClassPath=&Border, widget=Border, properties {brushColor=<rgba 0.0275 0.0275 0.0275 1>, padding=<ltrb 0 0 0 0>}}
  VerticalBox {widgetClassPath=&VerticalBox, parent=Border, widget=VerticalBox, slot {class=BorderSlot, properties {padding=<ltrb 0 0 0 0>}}}
  TitleTextBlock {
   widgetClassPath=&TextBlock
   bIsVariable=true
   parent=VerticalBox
   widget=TitleTextBlock
   properties {
    text=Title
    colorAndOpacity {colorUseRule=UseColor_Specified, specifiedColor=<rgba 0.5255 0.5255 0.5255 1>}
    font {
     bForceMonospaced=false
     bMaterialIsStencil=false
     fontMaterial=null
     fontObject=&Roboto
     letterSpacing=0
     monospacedWidth=1
     size=10
     skewAmount=0
     typefaceFontName=Bold
     outlineSettings {
      bApplyOutlineToDropShadows=false
      bMiteredCorners=false
      bSeparateFillAlpha=false
      outlineMaterial=null
      outlineSize=0
      outlineColor=<rgba 0 0 0 1>
     }
    }
   }
   slot {class=VerticalBoxSlot, properties {padding=<ltrb 5 5 5 5>}}
  }
  Text1HorizontalBox {
   widgetClassPath=&HorizontalBox
   bIsVariable=true
   parent=VerticalBox
   widget=Text1HorizontalBox
   slot {class=VerticalBoxSlot, properties {padding=<ltrb 5 0 0 0>}}
  }
  PathTitle1TextBlock {
   widgetClassPath=&TextBlock
   bIsVariable=true
   parent=Text1HorizontalBox
   widget=PathTitle1TextBlock
   properties {
    text=Info 1
    colorAndOpacity {colorUseRule=UseColor_Specified, specifiedColor=<rgba 0.5255 0.5255 0.5255 1>}
    font {
     bForceMonospaced=false
     bMaterialIsStencil=false
     fontMaterial=null
     fontObject=&Roboto
     letterSpacing=0
     monospacedWidth=1
     size=10
     skewAmount=0
     typefaceFontName=Bold
     outlineSettings {
      bApplyOutlineToDropShadows=false
      bMiteredCorners=false
      bSeparateFillAlpha=false
      outlineMaterial=null
      outlineSize=0
      outlineColor=<rgba 0 0 0 1>
     }
    }
   }
   slot {class=HorizontalBoxSlot, properties {padding=<ltrb 5 0 5 0>}}
  }
  Path1TextBlock {
   widgetClassPath=&TextBlock
   bIsVariable=true
   parent=Text1HorizontalBox
   widget=Path1TextBlock
   properties {
    justification=Right
    text=Text 1
    colorAndOpacity {colorUseRule=UseColor_Specified, specifiedColor=<rgba 0.5255 0.5255 0.5255 1>}
    font {
     bForceMonospaced=false
     bMaterialIsStencil=false
     fontMaterial=null
     fontObject=&Roboto
     letterSpacing=0
     monospacedWidth=1
     size=7.5
     skewAmount=0
     typefaceFontName=Regular
     outlineSettings {
      bApplyOutlineToDropShadows=false
      bMiteredCorners=false
      bSeparateFillAlpha=false
      outlineMaterial=null
      outlineSize=0
      outlineColor=<rgba 0 0 0 1>
     }
    }
   }
   slot {class=HorizontalBoxSlot, properties {verticalAlignment=VAlign_Center}}
  }
  Error1TextBlock {
   widgetClassPath=&TextBlock
   bIsVariable=true
   parent=VerticalBox
   widget=Error1TextBlock
   properties {
    text=Error Message
    visibility=Collapsed
    colorAndOpacity {colorUseRule=UseColor_Specified, specifiedColor=<rgba 1 0.0941 0.102 1>}
    font {
     bForceMonospaced=false
     bMaterialIsStencil=false
     fontMaterial=null
     fontObject=&Roboto
     letterSpacing=0
     monospacedWidth=1
     size=10
     skewAmount=0
     typefaceFontName=Bold
     outlineSettings {
      bApplyOutlineToDropShadows=false
      bMiteredCorners=false
      bSeparateFillAlpha=false
      outlineMaterial=null
      outlineSize=0
      outlineColor=<rgba 0 0 0 1>
     }
    }
   }
   slot {class=VerticalBoxSlot, properties {padding=<ltrb 10 2 0 0>}}
  }
  Text2HorizontalBox {
   widgetClassPath=&HorizontalBox
   bIsVariable=true
   parent=VerticalBox
   widget=Text2HorizontalBox
   slot {class=VerticalBoxSlot, properties {padding=<ltrb 5 0 0 0>}}
  }
  Title2TextBlock {
   widgetClassPath=&TextBlock
   bIsVariable=true
   parent=Text2HorizontalBox
   widget=Title2TextBlock
   properties {
    text=Info 2
    colorAndOpacity {colorUseRule=UseColor_Specified, specifiedColor=<rgba 0.5255 0.5255 0.5255 1>}
    font {
     bForceMonospaced=false
     bMaterialIsStencil=false
     fontMaterial=null
     fontObject=&Roboto
     letterSpacing=0
     monospacedWidth=1
     size=10
     skewAmount=0
     typefaceFontName=Bold
     outlineSettings {
      bApplyOutlineToDropShadows=false
      bMiteredCorners=false
      bSeparateFillAlpha=false
      outlineMaterial=null
      outlineSize=0
      outlineColor=<rgba 0 0 0 1>
     }
    }
   }
   slot {class=HorizontalBoxSlot, properties {padding=<ltrb 5 0 5 0>}}
  }
  Text2TextBlock {
   widgetClassPath=&TextBlock
   bIsVariable=true
   parent=Text2HorizontalBox
   widget=Text2TextBlock
   properties {
    justification=Right
    text=Text 2
    colorAndOpacity {colorUseRule=UseColor_Specified, specifiedColor=<rgba 0.5255 0.5255 0.5255 1>}
    font {
     bForceMonospaced=false
     bMaterialIsStencil=false
     fontMaterial=null
     fontObject=&Roboto
     letterSpacing=0
     monospacedWidth=1
     size=7.5
     skewAmount=0
     typefaceFontName=Italic
     outlineSettings {
      bApplyOutlineToDropShadows=false
      bMiteredCorners=false
      bSeparateFillAlpha=false
      outlineMaterial=null
      outlineSize=0
      outlineColor=<rgba 0 0 0 1>
     }
    }
   }
   slot {class=HorizontalBoxSlot, properties {verticalAlignment=VAlign_Center}}
  }
  Error2TextBlock {
   widgetClassPath=&TextBlock
   bIsVariable=true
   parent=VerticalBox
   widget=Error2TextBlock
   properties {
    text=Error Message
    visibility=Collapsed
    colorAndOpacity {colorUseRule=UseColor_Specified, specifiedColor=<rgba 1 0.0941 0.102 1>}
    font {
     bForceMonospaced=false
     bMaterialIsStencil=false
     fontMaterial=null
     fontObject=&Roboto
     letterSpacing=0
     monospacedWidth=1
     size=10
     skewAmount=0
     typefaceFontName=Bold
     outlineSettings {
      bApplyOutlineToDropShadows=false
      bMiteredCorners=false
      bSeparateFillAlpha=false
      outlineMaterial=null
      outlineSize=0
      outlineColor=<rgba 0 0 0 1>
     }
    }
   }
   slot {class=VerticalBoxSlot, properties {padding=<ltrb 10 2 0 0>}}
  }
 ]
}
