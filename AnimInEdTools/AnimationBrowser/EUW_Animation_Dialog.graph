EditorUtilityWidgetBlueprint /AnimInEdTools/AnimationBrowser/EUW_Animation_Dialog
graphs [
 Add Developer Options to Combobox {nodes [
   n1 = Add Developer Options to Combobox!() -> (then, Combo Box:Widget Object Reference, Name Type:Enum_FileNameBreakdown Enum=NewEnumerator2) {is=title K2Node_FunctionEntry}
   n2 = GetParent!(self:Widget Object Reference=@n1) -> (ReturnValue:Panel Widget Object Reference)
   > Branch(Condition="integer <= integer"(A=GetChildrenCount!(self:Panel Widget Object Reference=@n2) -> (ReturnValue:int=0), B=1))
   > n3 = Create Widget!(
    execute:Exec
    WorldContextObject:Object Reference
    Class:User Widget Class Reference=&/EUW_AddProjectEntry.EUW_AddProjectEntry_C
    OwningPlayer:Player Controller Object Reference
   ) -> (then:Exec, ReturnValue:EUW Add Project Entry Object Reference)
   > Set Animatable Creature!(
    execute
    self:EUW Add Project Entry Object Reference=@n3
    Project:BP Animation Browser Project Object Reference=Get Project() -> (Project:BP Animation Browser Project Object Reference)
    Animatable Creature:Animatable Creature Object Reference=Get Current Animatable Creature!() -> (Animatable Creatures:Animatable Creature Object Reference) {is=title K2Node_CallFunction}
    Name Type:Enum_FileNameBreakdown Enum=Get Name Type() -> (Name Type:Enum_FileNameBreakdown Enum=NewEnumerator2)
   ) -> (then) {is=title K2Node_CallFunction}
   > AddChild(self=@n2, Content=@n3)
   > Bind Event to Add Name To Data Asset(self:EUW Add Project Entry Object Reference=@n3, Delegate=Create Event()) {is=title K2Node_AddDelegate}
  ]}
 Add Name to DataAsset {nodes [
   n1 = Add Name to DataAsset!() -> (then, Data Asset:BP Animation Browser File Name Base Object Reference, Name:String, Name Type:Enum_FileNameBreakdown Enum=NewEnumerator2) {is=title K2Node_FunctionEntry}
   > n2 = Switch on Enum_FileNameBreakdown!(execute, Selection:Enum_FileNameBreakdown Enum=@n1.Name Type) -> (NewEnumerator2:Exec, NewEnumerator0:Exec, NewEnumerator1:Exec, NewEnumerator3:Exec) {is=title K2Node_SwitchEnum}
   Update Descriptor ComboBox(execute=@n2.NewEnumerator0)
   > Set Value!(
    execute
    self:EUW Custom Entry Combo Box Object Reference=Get DescriptorCustomEntryComboBox() -> (DescriptorCustomEntryComboBox:EUW Custom Entry Combo Box Object Reference)
    New Value:String=Get Name() -> (Name:String)
   ) -> (then) {is=title K2Node_CallFunction}
   Update Direction ComboBox(execute=@n2.NewEnumerator1)
   > SetSelectedOption!(
    execute
    self:"Searchable Combo Box (String) Object Reference"=Get DirectionComboBox() -> (DirectionComboBox:"Searchable Combo Box (String) Object Reference")
    Option:String=Get Name()
   ) -> (then)
   Update AnimSet ComboBox(execute=@n2.NewEnumerator2)
   > SetSelectedOption(self=Get AnimSetComboBox() -> (AnimSetComboBox:"Searchable Combo Box (String) Object Reference"), Option=Get Name())
   Update Variation ComboBox(execute=@n2.NewEnumerator3)
   > SetSelectedOption(self=Get VariationComboBox() -> (VariationComboBox:"Searchable Combo Box (String) Object Reference"), Option=Get Name())
   Get VariationComboBox()
   Get ComboBox!(self:EUW Custom Entry Combo Box Object Reference=Get DescriptorCustomEntryComboBox()) -> (ComboBox:"Searchable Combo Box (String) Object Reference")
  ]}
 Add Option {nodes [
   n1 = Inputs!() -> (Exec:Exec, Input:String)
   > n2 = Branch(Condition="Equal, Case Insensitive (String)"(A=@n1, B=Add Option))
   Outputs!(Passthrough:Exec=@n2.else, Add Option:Exec=@n2.then)
  ]}
 AnimSequence to Name Parts {nodes [
   AnimSequence to Name Parts!() -> (then, Anim Sequence Path:String, Animatable Creature:Animatable Creature Object Reference) {is=title K2Node_FunctionEntry}
   > n1 = Sequence!(execute) -> (then_0, then_1, then_2:Exec, then_3:Exec, then_4:Exec)
   n2 = Get Animatable Creature() -> (Animatable Creature:Animatable Creature Object Reference)
   n3 = Execute Python Script!(
    execute=@n1.then_2
    PythonScript="anim_set = ''\r\nname = ''\r\ndescriptor = ''\r\ndirection = ''\r\nvariation = ''\r\n\r\nvariations_lower = [x.lower() for x in variations]\r\ndescriptors_lower = [x.lower() for x in descriptors]\r\ndirections_lower = [x.lower() for x in directions]\r\nanim_sets_lower = [x.lower() for x in anim_sets]\r\n\r\n## Split by '_'\r\nsections = file_name.split('_')\r\n\r\nif len(sections) and sections[0] == asset:\r\n    sections = sections[1:] \r\n\r\nif len(sections) and sections[0] == plugin:\r\n    sections = sections[1:] \r\n\r\nremaining_sections = sections\r\n\r\nfor section in sections:\r\n    section_lower = section.lower()\r\n\r\n    if not variation:\r\n        try:\r\n            i = variations_lower.index(section_lower)\r\n            variation = variations[i]\r\n            remaining_sections.remove(section)\r\n            continue\r\n        except ValueError:\r\n            pass\r\n    if not direction:\r\n        try:\r\n            i = directions_lower.index(section_lower)\r\n            direction = directions[i]\r\n            remaining_sections.remove(section)\r\n            continue\r\n        except ValueError:\r\n            pass\r\n    if not descriptor:\r\n        try:\r\n            i = descriptors_lower.index(section_lower)\r\n            descriptor = descriptors[i]\r\n            remaining_sections.remove(section)\r\n            continue\r\n        except ValueError:\r\n            pass\r\n    if not anim_set:\r\n        try:\r\n            i = anim_sets_lower.index(section_lower)\r\n            anim_set = anim_sets[i]\r\n            remaining_sections.remove(section)\r\n            continue\r\n        except ValueError:\r\n            pass\r\n\r\nname = remaining_sections[-1]\r\nsuccess = len(remaining_sections) <= 1"
    asset:String=Get Asset!(self:Animatable Creature Object Reference=@n2) -> (Asset:String)
    plugin:String=Get Variant!(self:Animatable Creature Object Reference=@n2) -> (Variant:String)
    file_name:String=GetBaseFilename(InPath=Get Anim Sequence Path() -> (Anim Sequence Path:String), bRemovePath=true)
    variations:Array of Strings=Get Variations!(self:Animatable Creature Object Reference=Get Animatable Creature()) -> (Subfolders:Array of Strings) {is=title K2Node_CallFunction}
    directions:Array of Strings=Get Direction!(self:Animatable Creature Object Reference=Get Animatable Creature()) -> (Subfolders:Array of Strings) {is=title K2Node_CallFunction}
    descriptors:Array of Strings=Get Descriptors!(self:Animatable Creature Object Reference=Get Animatable Creature()) -> (Subfolders:Array of Strings) {is=title K2Node_CallFunction}
    anim_sets:Array of Strings=Get Anim Set!(self:Animatable Creature Object Reference=Get Animatable Creature()) -> (Subfolders:Array of Strings) {is=title K2Node_CallFunction}
   ) -> (then, ReturnValue, anim_set:String, name:String, descriptor:String, direction:String, variation:String, Success:bool=~) {inputs [asset,plugin,file_name,variations,directions,descriptors,anim_sets], outputs [anim_set,name,descriptor,direction,variation,Success]}
   > Return Node!(
    execute
    Subfolder:String=Get Subfolder Local() -> (Subfolder Local:String)
    Anim Set:String=@n3.anim_set
    Name:String=@n3.name
    Descriptor:String=@n3.descriptor
    Direction:String=@n3.direction
    Variation:String=@n3.variation
    Success:bool=@n3.Success
   )
   n4 = Get Anim Sequence Path()
   Set Subfolder Local(execute=@n1.then_0, Subfolder Local:String=GetBaseFilename(InPath=GetPath!(InPath:String=@n4) -> (ReturnValue:String), bRemovePath=true)) -> (Output_Get:String)
  ]}
 Developer ComboBox Filter {nodes [
   n1 = Developer ComboBox Filter!() -> (then, Input:String) {is=title K2Node_FunctionEntry}
   > Return Node!(execute, Output:String=SelectString(B=@n1, bPickA="Equal, Case Insensitive (String)"(A=@n1, B=Add Option)))
  ]}
 EventGraph {nodes [
   n1 = "On Selection Changed (ProjectComboBox)"!() -> (OutputDelegate, then, SelectedItem:String, SelectionType:ESelectInfo Enum) {is=title K2Node_ComponentBoundEvent}
   > n2 = For Each Loop with Break!(
    Exec:Exec
    Array:Array of BP Animation Browser Project Object References=Get All Projects!() -> (Projects:Array of BP Animation Browser Project Object References) {is=title K2Node_CallFunction}
    Break:Exec=@n3.then
   ) -> (LoopBody:Exec, Array Element:BP Animation Browser Project Object Reference, Array Index:int, Completed:Exec)
   > Branch(Condition="Equal Exactly (String)"(A=Get Name!(self:BP Animation Browser Project Object Reference=@n2) -> (Name), B=@n1.SelectedItem))
   > Set Project(Project:BP Animation Browser Project Object Reference=@n2) -> (Output_Get:BP Animation Browser Project Object Reference)
   > n3 = Set Options from Animatable Creature!(
    execute
    self
    Selected Creature:Animatable Creature Object Reference
    All Creatures:Array of Animatable Creature Object References=Get Animatable Creatures!() -> (Animatable Creatures:Array of Animatable Creature Object References) {is=title K2Node_CallFunction}
    Update Assets:bool=true
    Update Plugins:bool=true
    Set Plugin:bool=true
   ) -> (then) {is=title K2Node_CallFunction}
   n4 = "On Selection Changed (AnimSetComboBox)"!() -> (OutputDelegate, then, SelectedItem:String, SelectionType:ESelectInfo Enum) {is=title K2Node_ComponentBoundEvent}
   > n5 = Switch on ESelectInfo!(execute, Selection:ESelectInfo Enum=@n4.SelectionType) -> (OnKeyPress:Exec, OnNavigation:Exec, OnMouseClick:Exec, Direct:Exec) {is=title K2Node_SwitchEnum}
   n6 = Add Option!(Exec:Exec=[@n5.OnKeyPress, @n5.OnMouseClick, @n5.OnNavigation], Input:String=@n4.SelectedItem) -> (Passthrough:Exec, Add Option:Exec) {is=title K2Node_MacroInstance}
   Add Developer Options to Combobox!(execute=@n6.Add Option, self, Combo Box:Widget Object Reference=Get AnimSetComboBox(), Name Type:Enum_FileNameBreakdown Enum=NewEnumerator2) -> (then) {is=title K2Node_CallFunction}
   Update File Name from UI(execute=@n6.Passthrough)
   > Update Plugin Path from UI()
   > Remove Developer Option!(execute, self, Combo Box:Widget Object Reference=Get AnimSetComboBox()) -> (then)
   n7 = "On Selection Changed (VariationComboBox)"!() -> (OutputDelegate, then, SelectedItem:String, SelectionType:ESelectInfo Enum) {is=title K2Node_ComponentBoundEvent}
   > n8 = Add Option(Input=@n7.SelectedItem)
   Add Developer Options to Combobox(execute=@n8.Add Option, Combo Box=Get VariationComboBox(), Name Type=NewEnumerator3)
   Update File Name from UI(execute=@n8.Passthrough)
   > Update Plugin Path from UI()
   > Remove Developer Option(Combo Box=Get VariationComboBox())
   n9 = "On Selection Changed (DirectionComboBox)"!() -> (OutputDelegate, then, SelectedItem:String, SelectionType:ESelectInfo Enum) {is=title K2Node_ComponentBoundEvent}
   > n10 = Add Option(Input=@n9.SelectedItem)
   Add Developer Options to Combobox(execute=@n10.Add Option, Combo Box=Get DirectionComboBox(), Name Type=NewEnumerator1)
   Update File Name from UI(execute=@n10.Passthrough)
   > Update Plugin Path from UI()
   > Remove Developer Option(Combo Box=Get DirectionComboBox())
   n11 = "On Selection Changed (AssetComboBox)"!() -> (OutputDelegate, then, SelectedItem:String, SelectionType:ESelectInfo Enum) {is=title K2Node_ComponentBoundEvent}
   > n12 = Switch on ESelectInfo(Selection=@n11.SelectionType)
   n13 = Branch(
    execute=@n12.OnNavigation
    Condition=IsOpen!(self:"Searchable Combo Box (String) Object Reference"=Get AssetComboBox() -> (AssetComboBox:"Searchable Combo Box (String) Object Reference")) -> (ReturnValue:bool)
   )
   n14 = Get Project()
   Set Options from Animatable Creature(
    execute=[@n13.then, @n12.OnKeyPress, @n12.OnMouseClick]
    Selected Creature=Random Array Item!(TargetArray:Array of Animatable Creature Object References=Get Variations for Creature!(Creature:String=@n11.SelectedItem, Project:BP Animation Browser Project Object Reference=@n14, __WorldContext:Object Reference) -> (Animatable Creatures:Array of Animatable Creature Object References) {is=title K2Node_CallFunction}) -> (OutItem:Animatable Creature Object Reference, OutIndex:int=0) {is=title K2Node_CallArrayFunction}
    All Creatures=Get Animatable Creatures()
    Update Assets=false
    Update Plugins=true
    Set Plugin=true
   )
   n15 = "On Value Changed (DescriptorCustomEntryComboBox)"!() -> (OutputDelegate, then, Value:String) {is=title K2Node_ComponentBoundEvent}
   > n16 = Add Option(Input=@n15.Value)
   Add Developer Options to Combobox(execute=@n16.Add Option, Combo Box=Get DescriptorCustomEntryComboBox(), Name Type=NewEnumerator0)
   RetriggerableDelay!(execute=@n16.Passthrough, WorldContextObject:Object Reference, Duration=0.7) -> (then)
   > Update File Name from UI()
   > Update Plugin Path from UI()
   > Remove Developer Option(Combo Box=Get DescriptorCustomEntryComboBox())
   "On Clicked (CreateButton)"() {is=title K2Node_ComponentBoundEvent}
   > n17 = Validate UI Input!(execute, self) -> (then, Is Valid:bool)
   > Branch(Condition=@n17)
   > n18 = Generate Anim and Level Sequence Paths!(execute, self) -> (
    then
    Anim Sequence:String
    Level Sequence:String
    Animatable Creatures:Animatable Creature Object Reference
    Animation Path:String
    Source Path:String
   )
   > Call Generate Animation Paths!(
    execute:Exec
    self:EUW Animation Dialog Object Reference
    AnimSequence Path:String=@n18
    LevelSequence Path:String=@n18.Level Sequence
    Animatable Creature:Animatable Creature Object Reference=@n18.Animatable Creatures
   ) -> (then:Exec) {is=title K2Node_CallDelegate}
   > RemoveFromParent()
   "On Clicked (CancelButton)"() {is=title K2Node_ComponentBoundEvent}
   > RemoveFromParent()
   n19 = "On Selection Changed (PluginComboBox)"!() -> (OutputDelegate, then, SelectedItem:String, SelectionType:ESelectInfo Enum) {is=title K2Node_ComponentBoundEvent}
   > n20 = Switch on ESelectInfo(Selection=@n19.SelectionType)
   Set Options from Animatable Creature(
    execute=[@n21.then, @n20.OnKeyPress, @n20.OnMouseClick]
    Selected Creature=Find Animatable Creature!(
     Asset:String=Get Asset!() -> (Value:String) {is=title K2Node_CallFunction}
     Variation:String=@n19.SelectedItem
     Project:BP Animation Browser Project Object Reference=Get Project()
     __WorldContext:Object Reference
    ) -> (Animatable Creatures:Animatable Creature Object Reference)
    All Creatures=Get Animatable Creatures()
    Update Assets=false
    Update Plugins=false
    Set Plugin=true
   )
   n21 = Branch(execute=@n20.OnNavigation, Condition=IsOpen(self=Get PluginComboBox() -> (PluginComboBox:"Searchable Combo Box (String) Object Reference")))
   "On Text Changed (NameTextBox)"!() -> (OutputDelegate, then, Text:"Text (by ref)") {is=title K2Node_ComponentBoundEvent}
   > RetriggerableDelay(Duration=0.7)
   > Update File Name from UI()
   > Update Plugin Path from UI()
   "On Opening (ProjectComboBox)"() {is=title K2Node_ComponentBoundEvent}
   > ClearOptions(self:"Searchable Combo Box (String) Object Reference"=Get ProjectComboBox() -> (ProjectComboBox:"Searchable Combo Box (String) Object Reference"))
   > n22 = For Each Loop(Array:Array of BP Animation Browser Project Object References=Get All Projects()) -> (Array Element:BP Animation Browser Project Object Reference)
   > AddOption!(execute, self:"Searchable Combo Box (String) Object Reference"=Get ProjectComboBox(), Option:String=Get Name(self=@n22)) -> (then)
   n23 = "On Selection Changed (Combo Box) (SubfolderCustomEntryComboBox)"!() -> (OutputDelegate, then, Selected Item:String, Selection Type:ESelectInfo Enum) {is=title K2Node_ComponentBoundEvent}
   > n24 = Switch on ESelectInfo(Selection=@n23.Selection Type)
   n25 = Get Current Animatable Creature()
   n26 = Update Plugin Path!(
    execute=[@n24.OnKeyPress, @n24.OnMouseClick, @n24.OnNavigation]
    self
    Subfolder:String=@n23.Selected Item
    Animatable Creature:Animatable Creature Object Reference=@n25
   ) -> (then)
   Update File Name from UI(execute=[@n28.then, @n26.then])
   > AnimSet Developer()
   n27 = "On Value Changed (SubfolderCustomEntryComboBox)"!() -> (OutputDelegate, then, Value:String) {is=title K2Node_ComponentBoundEvent}
   n28 = Update Plugin Path(execute=[@n27.then, @n29.then], Subfolder=@n27.Value, Animatable Creature=@n25)
   n29 = "On Text Changed (Text Box) (SubfolderCustomEntryComboBox)"!() -> (OutputDelegate, then, Text:Text) {is=title K2Node_ComponentBoundEvent}
   Update Descriptor ComboBox() {is=title K2Node_CustomEvent}
   > Set Combo Box Options!(
    execute
    self
    Combo Box:"Searchable Combo Box (String) Object Reference"=Get ComboBox(self=Get DescriptorCustomEntryComboBox())
    Options:Array of Strings=Get Descriptors(self=Get Current Animatable Creature())
    Add Developer Options:bool=true
    Hide if there are no options:bool=true
   ) -> (then) {is=title K2Node_CallFunction}
   n30 = Set Initial Settings!() -> (
    OutputDelegate
    then
    Project:BP Animation Browser Project Object Reference
    Selected Creature:Animatable Creature Object Reference
    Developer Mode:bool=~
    Set Plugin:bool=~
    Animatable Creatures:Array of Animatable Creature Object References
    Show Project Dialog:bool=~
    Override Anim Path Validation:bool=~
   ) {is=title K2Node_CustomEvent}
   > Set Project(Project=@n30.Project)
   > Set Animatable Creatures!(execute, self, Animatable Creatures:Array of Animatable Creature Object References=@n30.Animatable Creatures) -> (then) {is=title K2Node_CallFunction}
   > Set Developer Mode(Developer Mode:bool=@n30.Developer Mode) -> (Output_Get:bool)
   > Set Options from Animatable Creature(
    Selected Creature=@n30.Selected Creature
    All Creatures=Get Animatable Creatures()
    Update Assets=true
    Update Plugins=true
    Set Plugin=@n30.Set Plugin
   )
   > Set Override Validation(Override Validation:bool=@n30.Override Anim Path Validation) -> (Output_Get:bool)
   > SetVisibility(
    self=Get ProjectHorizontalBox() -> (ProjectHorizontalBox:Horizontal Box Object Reference)
    InVisibility=Select(Option 0=Collapsed, Option 1=Visible, Index=@n30.Show Project Dialog)
   )
   n31 = Get Current Animatable Creature()
   > SetVisibility(
    self=Get EditorUtilityButton!(self:EUW Custom Entry Combo Box Object Reference=Get SubfolderCustomEntryComboBox() -> (SubfolderCustomEntryComboBox:EUW Custom Entry Combo Box Object Reference)) -> (EditorUtilityButton:Editor Utility Button Object Reference)
    InVisibility=Select(
     Option 0=Collapsed
     Option 1=Visible
     Index=Get Populate Subfolders From Disk!(self:Animatable Creature Object Reference=@n31) -> (Populate Subfolders From Disk:bool)
    )
   )
   Update Variation ComboBox() {is=title K2Node_CustomEvent}
   > Set Combo Box Options(
    Combo Box=Get VariationComboBox()
    Options=Get Variations(self=Get Current Animatable Creature())
    Add Developer Options=true
    Hide if there are no options=true
   )
   Update Direction ComboBox() {is=title K2Node_CustomEvent}
   > Set Combo Box Options(
    Combo Box=Get DirectionComboBox()
    Options=Get Direction(self=Get Current Animatable Creature())
    Add Developer Options=true
    Hide if there are no options=true
   )
   Update AnimSet ComboBox() {is=title K2Node_CustomEvent}
   > n32 = Sequence!(execute) -> (then_0, then_1, then_2)
   > Set Combo Box Options(
    Combo Box=Get AnimSetComboBox()
    Options=Get Anim Set(self=Get Current Animatable Creature())
    Add Developer Options=true
    Hide if there are no options=true
   )
   n33 = Get CurrentUser() -> (CurrentUser:String)
   Branch(execute=[@n37.then, @n32.then_1], Condition=NOT Boolean(A=IsEmpty(InString=@n33)))
   n34 = Get ComboBox(self=Get SubfolderCustomEntryComboBox())
   > n35 = Branch(Condition="Equal, Case Insensitive (String)"(A=GetSelectedOption!(self:"Searchable Combo Box (String) Object Reference"=@n34) -> (ReturnValue:String), B=Developer))
   n36 = Camel Case String!(String:String=Get CurrentUser(), __WorldContext:Object Reference) -> (New String:String)
   > AddOption(self=Get AnimSetComboBox(), Option=@n36)
   RemoveOption!(execute=@n35.else, self:"Searchable Combo Box (String) Object Reference"=Get AnimSetComboBox(), Option:String=@n36) -> (then, ReturnValue:bool)
   n37 = AnimSet Developer() {is=title K2Node_CustomEvent}
   Event Construct()
   > ExecutePythonCommand(PythonCommand=from utilities import object_utils)
   > Set Display Info!(
    execute
    self:EUW Animation Dialog Border Info Object Reference=Get FeedbackWidget() -> (FeedbackWidget:EUW Animation Dialog Border Info Object Reference)
    Title:Text="LOCTEXT(\"New File Name\")"
    Title 1:Text
    Text 1:Text
    Title 2:Text
    Text 2:Text
    Error 1:Text
    Error 2:Text
   ) -> (then) {is=title K2Node_CallFunction}
  ]}
 Generate Anim and Level Sequence Paths {nodes [
   Generate Anim and Level Sequence Paths() {is=title K2Node_FunctionEntry}
   n1 = Get Current Animatable Creature()
   > n2 = Is Valid(InputObject=@n1)
   Return Node!(
    execute=@n2.Is Not Valid
    Anim Sequence:String
    Level Sequence:String
    Animatable Creatures:Animatable Creature Object Reference
    Animation Path:String
    Source Path:String
   )
   n3 = Get RemoveSpacesOnName!(self:Animatable Creature Object Reference=Get Current Animatable Creature()) -> (RemoveSpacesOnName:bool)
   n4 = Get Naming Convention!(self:Animatable Creature Object Reference=@n1) -> (ReturnValue:EU Animation Naming Convention Object Reference) {is=title K2Node_CallFunction}
   n5 = Get Path!(
    self:EU Animation Naming Convention Object Reference=@n4
    Token Map:Map of Strings to Strings=Get Path Parts Map From UI!() -> (Path Parts Map:Map of Strings to Strings) {is=title K2Node_CallFunction}
    Path Pattern:String=Get Animation Path!(self:Animatable Creature Object Reference=@n1) -> (Animation Path:String)
    Remove Spaces On Name:bool=@n3
   ) -> (File Path:String, Folder Path:String, File Name:String) {is=title K2Node_CallFunction}
   n6 = Get Path(
    self=@n4
    Token Map=Get Path Parts Map From UI()
    Path Pattern=Get Source Path!(self:Animatable Creature Object Reference=@n1) -> (Source Path:String)
    Remove Spaces On Name=@n3
   )
   > Return Node(Anim Sequence=@n5, Level Sequence=@n6, Animatable Creatures=@n1, Animation Path=@n5.Folder Path, Source Path=@n6.Folder Path)
  ]}
 Generate Animation Paths {nodes [Generate Animation Paths!() -> (then, AnimSequence Path:String, LevelSequence Path:String, Animatable Creature:Animatable Creature Object Reference) {is=title K2Node_FunctionEntry}]}
 Get All Projects {nodes [
   Get All Projects() {is=title K2Node_FunctionEntry}
   > Clear(TargetArray:Array of BP Animation Browser Project Object References=Get Local Projects() -> (Local Projects:Array of BP Animation Browser Project Object References))
   > n1 = GetAssetsByClass!(
    execute
    self:Asset Registry Interface=GetAssetRegistry()
    ClassPathName:Top Level Asset Path Structure=GetSoftClassTopLevelAssetPath!(SoftClassReference:Object Soft Class Reference=/AnimInEdTools/AnimationBrowser/ProjectSettings/BP_AnimationBrowser_Project.BP_AnimationBrowser_Project_C) -> (ReturnValue:Top Level Asset Path Structure)
    bSearchSubClasses:bool=false
   ) -> (then, OutAssetData:Array of Asset Data Structures, ReturnValue:bool)
   > n2 = For Each Loop(Array:Array of Asset Data Structures=@n1) -> (Array Element:Asset Data Structure)
   Return Node!(execute=@n2.Completed, Projects:Array of BP Animation Browser Project Object References=Get Local Projects())
   n3 = Break Asset Data(AssetData=@n2)
   > n4 = LoadAsset(AssetPath=GetBaseFilename(InPath="To String (Name)"(InName=@n3), bRemovePath=false))
   > n5 = Cast To BP_AnimationBrowser_Project!(execute, Object=@n4) -> (then, CastFailed, AsBP Animation Browser Project:BP Animation Browser Project Object Reference) {is=title K2Node_DynamicCast}
   > Add(
    TargetArray:Array of BP Animation Browser Project Object References=Get Local Projects()
    NewItem:"BP Animation Browser Project Object Reference (by ref)"=@n5
   ) {is=title K2Node_CallArrayFunction}
  ]}
 Get AnimSet {nodes [
   Get AnimSet() {is=title K2Node_FunctionEntry}
   n1 = Get AnimSetComboBox()
   > Return Node!(execute, Value:String=Developer ComboBox Filter!(Input:String=GetSelectedOption(self=@n1)) -> (Output:String) {is=title K2Node_CallFunction})
  ]}
 Get Animatable Creatures {nodes [
   Get Animatable Creatures() {is=title K2Node_FunctionEntry}
   > n1 = Branch(Condition=Is Empty!(TargetArray:Array of Animatable Creature Object References=Get Animatable Creatures() -> (Animatable Creatures:Array of Animatable Creature Object References) {is=title K2Node_VariableGet}) -> (ReturnValue:bool))
   > Return Node!(
    execute
    Animatable Creatures:Array of Animatable Creature Object References=Get Animatable Creatures!(Project:BP Animation Browser Project Object Reference=Get Project(), bSearchSubClasses:bool=true, __WorldContext:Object Reference) -> (Animatable Creatures) {is=title K2Node_CallFunction}
   )
   Return Node(execute=@n1.else, Animatable Creatures=Get Animatable Creatures() {is=title K2Node_VariableGet})
  ]}
 Get Asset {nodes [
   Get Asset() {is=title K2Node_FunctionEntry}
   n1 = Get AssetComboBox()
   > Return Node!(execute, Value:String=Developer ComboBox Filter(Input=GetSelectedOption(self=@n1)))
  ]}
 Get Current Animatable Creature {nodes [
   Get Current Animatable Creature() {is=title K2Node_FunctionEntry}
   n1 = Get PluginComboBox()
   > Return Node!(
    execute
    Animatable Creatures:Animatable Creature Object Reference=Find Animatable Creature(Asset=Get Asset() {is=title K2Node_CallFunction}, Variation=GetSelectedOption(self=@n1), Project=Get Project())
   )
  ]}
 Get Descriptor {nodes [
   Get Descriptor() {is=title K2Node_FunctionEntry}
   n1 = Get DescriptorCustomEntryComboBox()
   > Return Node!(
    execute
    Value:String=Developer ComboBox Filter(Input=Get Value!(self:EUW Custom Entry Combo Box Object Reference=@n1) -> (Value:String) {is=title K2Node_CallFunction})
   )
  ]}
 Get Direction {nodes [
   Get Direction() {is=title K2Node_FunctionEntry}
   n1 = Get DirectionComboBox()
   > Return Node(Value=Developer ComboBox Filter(Input=GetSelectedOption(self=@n1)))
  ]}
 Get Name {nodes [
   Get Name() {is=title K2Node_FunctionEntry}
   n1 = Get NameTextBox() -> (NameTextBox:Text Box Object Reference)
   > Return Node(Value="To String (Text)"(InText=Get Text(self=@n1)))
  ]}
 Get Path Parts Map From UI {nodes [
   Get Path Parts Map From UI() {is=title K2Node_FunctionEntry}
   n1 = Get PluginComboBox()
   n2 = Get SubfolderCustomEntryComboBox()
   n3 = GetSelectedOption(self=Get ComboBox(self=@n2))
   n4 = Get NameTextBox!(self:EUW Custom Entry Combo Box Object Reference=@n2) -> (NameTextBox)
   n5 = "To String (Text)"(InText="GetText (Text Box)"!(self:Text Box Object Reference=@n4) -> (ReturnValue:Text))
   n6 = IsVisible(self=@n4)
   > Return Node!(
    execute
    Path Parts Map:Map of Strings to Strings=Get Path Parts Map!(
     Asset:String=Get Asset()
     Animation Set:String=Get AnimSet!() -> (Value:String) {is=title K2Node_CallFunction}
     Animation Name:String=Get Name!() -> (Value:String) {is=title K2Node_CallFunction}
     Descriptor:String=Get Descriptor!() -> (Value:String) {is=title K2Node_CallFunction}
     Direction:String=Get Direction!() -> (Value:String) {is=title K2Node_CallFunction}
     Subfolder:String=Select(Option 0:String=@n3, Option 1:String=@n5, Index=@n6) -> (ReturnValue:String)
     Variation:String=Get Variation!() -> (Value:String) {is=title K2Node_CallFunction}
     Plugin:String=GetSelectedOption(self=@n1)
     __WorldContext:Object Reference
    ) -> (Output Path Parts Map:Map of Strings to Strings) {is=title K2Node_CallFunction}
   )
  ]}
 Get Variation {nodes [
   Get Variation() {is=title K2Node_FunctionEntry}
   n1 = Get VariationComboBox()
   > Return Node!(execute, Value:String=Developer ComboBox Filter(Input=GetSelectedOption(self=@n1)))
  ]}
 Override TextBox {nodes [
   n1 = Override TextBox!() -> (then, TextBlock:Text Object Reference, String:String) {is=title K2Node_FunctionEntry}
   > n2 = Branch(Condition=IsEmpty(InString=@n1.String))
   "SetText (Text)"(execute=@n2.else, self=@n1, InText="To Text (String)"(InString=@n1.String))
  ]}
 Remove All Developer Options {nodes [
   Remove All Developer Options() {is=title K2Node_FunctionEntry}
   > n1 = For Each Loop(Array:Array of Widget Object References=Make Array!(
     "[0]":Widget Object Reference=Get AnimSetComboBox()
     "[1]":Widget Object Reference=Get DescriptorCustomEntryComboBox()
     "[2]":Widget Object Reference=Get DirectionComboBox()
     "[3]":Widget Object Reference=Get VariationComboBox()
    ) -> (Array:Array of Widget Object References)) -> (Array Element:Widget Object Reference)
   > Remove Developer Option(Combo Box=@n1)
  ]}
 Remove Developer Option {nodes [
   n1 = Remove Developer Option!() -> (then, Combo Box:Widget Object Reference) {is=title K2Node_FunctionEntry}
   n2 = GetParent(self=@n1)
   > Branch(Condition="integer > integer"(A=GetChildrenCount(self=@n2) -> (ReturnValue=0), B=1))
   > RemoveChildAt!(execute, self:Panel Widget Object Reference=@n2, Index:int=1) -> (then, ReturnValue:bool)
  ]}
 Set Animatable Creatures {nodes [
   n1 = Set Animatable Creatures!() -> (then, Animatable Creatures:Array of Animatable Creature Object References) {is=title K2Node_FunctionEntry}
   > Set Animatable Creatures(Animatable Creatures:Array of Animatable Creature Object References=@n1) -> (Output_Get:Array of Animatable Creature Object References) {is=title K2Node_VariableSet}
  ]}
 Set Combo Box Options {nodes [
   n1 = Set Combo Box Options!() -> (
    then
    Combo Box:"Searchable Combo Box (String) Object Reference"
    Options:Array of Strings
    Add Developer Options:bool=~
    Hide if there are no options:bool=true
   ) {is=title K2Node_FunctionEntry}
   > Set Previous Selection(Previous Selection:String=GetSelectedOption(self=@n1)) -> (Output_Get:String)
   > ClearOptions(self=Get Combo Box() -> (Combo Box:"Searchable Combo Box (String) Object Reference"))
   > n2 = Sequence!(execute) -> (then_0, then_1, then_2, then_3:Exec, then_4:Exec)
   n3 = NOT Boolean(A=Is Empty(TargetArray:Array of Strings=Get Options() -> (Options:Array of Strings)))
   n4 = Get Developer Mode() -> (Developer Mode:bool)
   > Branch(Condition=AND Boolean(A=Get Add Developer Options() -> (Add Developer Options:bool), B=OR Boolean(A=@n3, B=@n4)))
   > AddOption(self=Get Combo Box())
   n5 = For Each Loop(Exec=@n2.then_1, Array:Array of Strings=Get Options()) -> (Array Element:String)
   > AddOption(self=Get Combo Box(), Option=@n5)
   Branch(execute=@n2.then_2, Condition=AND Boolean(A=Get Developer Mode(), B=Get Add Developer Options()))
   > AddOption(self=Get Combo Box(), Option=Add Option)
   n6 = Branch(execute=@n2.then_3, Condition=IsEmpty(InString=Get Previous Selection() -> (Previous Selection:String)))
   n7 = Branch(
    execute=@n6.else
    Condition="Not Equal, Case Insenstive (String)"!(A:String=Get Previous Selection(), B:String=Add Option) -> (ReturnValue:bool)
   )
   n8 = Get Combo Box()
   n9 = GetSelectedIndex!(self:"Searchable Combo Box (String) Object Reference"=@n8) -> (ReturnValue:int=0)
   n10 = GetOptionCount!(self:"Searchable Combo Box (String) Object Reference"=@n8) -> (ReturnValue:int=0)
   > SetSelectedIndex!(
    execute
    self:"Searchable Combo Box (String) Object Reference"=Get Combo Box()
    Index:int=SelectInt(A=int - int(A=@n10), B=0, bPickA="Equal (Integer)"(A=@n9))
   ) -> (then)
   > SetSelectedOption(self=Get Combo Box(), Option=Get Previous Selection())
   ClearSelection(execute=@n7.else, self:"Searchable Combo Box (String) Object Reference"=Get Combo Box())
   n11 = GetOptionCount(self=Get Combo Box()) -> (ReturnValue=0)
   Branch(
    execute=@n2.then_4
    Condition=AND Boolean(A="Equal (Integer)"(A=@n11), B=Get Hide if there are no options() -> (Hide if there are no options:bool))
   )
   n12 = Get Combo Box()
   > SetVisibility(self=GetParent(self=GetParent(self=@n12)), InVisibility=Collapsed)
  ]}
 Set Options from Animatable Creature {nodes [
   n1 = Set Options from Animatable Creature!() -> (
    then
    Selected Creature:Animatable Creature Object Reference
    All Creatures:Array of Animatable Creature Object References
    Update Assets:bool=~
    Update Plugins:bool=~
    Set Plugin:bool=true
   ) {is=title K2Node_FunctionEntry}
   > n2 = Sequence!(execute) -> (then_0, then_1, then_2, then_3, then_4, then_5:Exec, then_6:Exec, then_7:Exec)
   > n3 = Set Naming Settings(Naming Settings:Enum_AC_Naming Enum=Get Naming Settings!(self:Animatable Creature Object Reference=Get Selected Creature() -> (Selected Creature:Animatable Creature Object Reference)) -> (Naming Settings:Enum_AC_Naming Enum=NewEnumerator0)) -> (Output_Get:Enum_AC_Naming Enum=NewEnumerator0)
   > n4 = Switch on Enum_AC_Naming!(execute, Selection:Enum_AC_Naming Enum=@n3) -> (NewEnumerator0:Exec, NewEnumerator1:Exec, NewEnumerator2:Exec) {is=title K2Node_SwitchEnum}
   SetVisibility(
    execute=[@n4.NewEnumerator0, @n4.NewEnumerator1]
    self=[
     Get VariationHorizontalBox() -> (VariationHorizontalBox:Horizontal Box Object Reference)
     Get DirectionHorizontalBox() -> (DirectionHorizontalBox:Horizontal Box Object Reference)
     Get DescriptorHorizontalBox() -> (DescriptorHorizontalBox:Horizontal Box Object Reference)
     Get AnimSetHorizontalBox() -> (AnimSetHorizontalBox:Horizontal Box Object Reference)
    ]
    InVisibility=Visible
   )
   ClearOptions(execute=@n4.NewEnumerator2, self=[Get VariationComboBox(), Get DirectionComboBox(), Get AnimSetComboBox()])
   > Clear(self:EUW Custom Entry Combo Box Object Reference=Get DescriptorCustomEntryComboBox()) {is=title K2Node_CallFunction}
   > SetVisibility(
    self=[Get VariationHorizontalBox(), Get DirectionHorizontalBox(), Get DescriptorHorizontalBox(), Get AnimSetHorizontalBox()]
    InVisibility=Collapsed
   )
   Clear(execute=@n2.then_1, TargetArray:Array of Strings=Get Assets() -> (Assets:Array of Strings)) {is=title K2Node_CallArrayFunction}
   > Clear(TargetArray=Get Variants() -> (Variants:Array of Strings))
   n5 = For Each Loop(Exec=@n2.then_2, Array:Array of Animatable Creature Object References=@n1.All Creatures) -> (Array Element:Animatable Creature Object Reference)
   SortStringArray!(execute=@n5.Completed, TargetArray:Array of Strings=Get Assets(), bStableSort:bool=false, SortOrder:EArraySortOrder Enum=Ascending) -> (then)
   > SortStringArray(TargetArray=Get Variants(), bStableSort=false, SortOrder=Ascending)
   n6 = Sequence!(execute=@n5.LoopBody) -> (then_0, then_1)
   > Branch(Condition=Get Update Assets() -> (Update Assets:bool))
   > Add Unique(TargetArray=Get Assets(), NewItem=Get Asset(self=@n5) {is=title K2Node_VariableGet})
   n7 = Get Selected Creature()
   Branch(execute=@n6.then_1, Condition="Equal, Case Insensitive (String)"(A=Get Asset(self=@n7), B=Get Asset(self=@n5)))
   > Add Unique(TargetArray=Get Variants(), NewItem=Get Variant(self=@n5))
   Branch(execute=@n2.then_3, Condition=Get Update Assets())
   > Set Combo Box Options(Combo Box=Get AssetComboBox(), Options=Get Assets(), Add Developer Options=false, Hide if there are no options=false) {is=title K2Node_CallFunction}
   Branch(execute=@n2.then_4, Condition=Get Update Plugins() -> (Update Plugins:bool))
   > Set Combo Box Options(Combo Box=Get PluginComboBox(), Options=Get Variants(), Add Developer Options=false, Hide if there are no options=false)
   SetSelectedOption(execute=@n2.then_6, self=Get AssetComboBox(), Option=Get Asset(self=Get Selected Creature()))
   > n8 = Branch(Condition=Get Set Plugin() -> (Set Plugin:bool))
   > n9 = SetSelectedOption(self=Get PluginComboBox(), Option=Get Variant(self=Get Selected Creature()))
   n10 = "Equal (Integer)"(A=GetSelectedIndex(self=Get PluginComboBox()) -> (ReturnValue=0), B=-1)
   n11 = Select(Option 0:ESlateVisibility Enum=Visible, Option 1:ESlateVisibility Enum=Collapsed, Index=@n10) -> (ReturnValue:ESlateVisibility Enum)
   SetVisibility(execute=[@n9.then, @n8.else], self=Get AssetNameBorder() -> (AssetNameBorder:Border Object Reference), InVisibility=@n11)
   > SetVisibility(self=Get FeedbackWidget(), InVisibility=@n11)
   > n12 = Branch(Condition=@n10)
   Update AnimSet ComboBox(execute=@n12.else) {is=title K2Node_CallFunction}
   > Update Descriptor ComboBox() {is=title K2Node_CallFunction}
   > Update Direction ComboBox() {is=title K2Node_CallFunction}
   > Update Variation ComboBox() {is=title K2Node_CallFunction}
   > Update File Name from UI()
   > Update Plugin Path from UI()
   > Remove All Developer Options() {is=title K2Node_CallFunction}
   n13 = Get Selected Creature()
   n14 = Get Populate Subfolders From Disk(self=@n13)
   n15 = Branch(execute=@n2.then_5, Condition=@n14)
   > Clear(TargetArray=Get Populated Subfolders() -> (Populated Subfolders:Array of Strings))
   > n16 = ListAssets!(
    execute
    DirectoryPath:String=Split(SourceString=Get Animation Path(self=@n13), InStr="/{Subfolder}", SearchDir=FromStart)
    bRecursive:bool=false
    bIncludeFolder:bool=true
   ) -> (then, ReturnValue:Array of Strings)
   > n17 = For Each Loop(Array:Array of Strings=@n16) -> (Array Element:String)
   Set Combo Box Options(
    execute=[@n15.else, @n17.Completed]
    Combo Box=Get ComboBox(self=Get SubfolderCustomEntryComboBox())
    Options=Select(
     Option 0:Array of Strings=Get Subfolders!(self:Animatable Creature Object Reference=@n13) -> (Subfolders:Array of Strings) {is=title K2Node_CallFunction}
     Option 1:Array of Strings=Get Populated Subfolders()
     Index=@n14
    ) -> (ReturnValue:Array of Strings)
    Add Developer Options=false
    Hide if there are no options=false
   )
   n18 = DoesDirectoryExist!(execute=@n17.LoopBody, DirectoryPath:String=@n17) -> (then, ReturnValue:bool)
   > Branch(Condition=@n18)
   > Add Unique(
    TargetArray=Get Populated Subfolders()
    NewItem=GetCleanFilename!(InPath:String=Split(SourceString=@n17, InStr=/, SearchDir=FromEnd)) -> (ReturnValue:String)
   )
  ]}
 Set Plugin Error {nodes [
   n1 = Set Plugin Error!() -> (then, Text:Text, Error Text Block:Text Object Reference, Text Block:Widget Object Reference) {is=title K2Node_FunctionEntry}
   > Set Text(Text=@n1, self=@n1.Error Text Block)
   n2 = TextIsEmpty(InText=@n1)
   > SetVisibility(
    self=@n1.Error Text Block
    InVisibility=Select(Option 0:ESlateVisibility Enum=Visible, Option 1:ESlateVisibility Enum=Collapsed, Index=@n2) -> (ReturnValue:ESlateVisibility Enum)
   )
   > SetVisibility(self=@n1.Text Block, InVisibility=Select(Option 0=Collapsed, Option 1=Visible, Index=@n2))
  ]}
 Set Subfolder {nodes [
   n1 = Set Subfolder!() -> (then, Option:String) {is=title K2Node_FunctionEntry}
   > SetSelectedOption(self=Get ComboBox(self=Get SubfolderCustomEntryComboBox()), Option=@n1)
   > Update Plugin Path from UI()
   Validate UI Input()
  ]}
 Set Title and Button Text {nodes [
   n1 = Set Title and Button Text!() -> (then, Title:String, Create Button:String, Description:String) {is=title K2Node_FunctionEntry}
   > Set Text(Text="To Text (String)"(InString=@n1), self=Get TitleTextBlock() -> (TitleTextBlock:Text Object Reference))
   > Set Text(Text="To Text (String)"(InString=@n1.Create Button), self=Get CreateButtonTextBlock() -> (CreateButtonTextBlock:Text Object Reference))
   n2 = Get Description() -> (Description:String)
   n3 = Get Description() -> (Description:Text Object Reference)
   > SetVisibility(self=@n3, InVisibility=Select(Option 0=Visible, Option 1=Collapsed, Index=IsEmpty(InString=@n2)))
   > Set Text(Text="To Text (String)"(InString=@n2), self=@n3)
  ]}
 Update Field Names {nodes [
   n1 = Update Field Names!() -> (then, Asset:String, Plugin:String, Subfolder:String, AnimationSet:String, Description:String, Direction:String, Mood/Status:String) {is=title K2Node_FunctionEntry}
   > Override TextBox!(execute, self, TextBlock:Text Object Reference=Get AssetTextBlock() -> (AssetTextBlock:Text Object Reference), String:String=@n1) -> (then) {is=title K2Node_CallFunction}
   > Override TextBox(TextBlock=Get PluginTextBlock() -> (PluginTextBlock:Text Object Reference), String=@n1.Plugin)
   > Override TextBox(TextBlock=Get SubfolderTextBlock() -> (SubfolderTextBlock:Text Object Reference), String=@n1.Subfolder)
   > Override TextBox(TextBlock=Get CategoryTextBlock() -> (CategoryTextBlock:Text Object Reference), String=@n1.AnimationSet)
   > Override TextBox(TextBlock=Get DescriptorTextBlock() -> (DescriptorTextBlock:Text Object Reference), String=@n1.Description)
   > Override TextBox(TextBlock=Get DirectionTextBlock() -> (DirectionTextBlock:Text Object Reference), String=@n1.Direction)
   > Override TextBox(TextBlock=Get VariationTextBlock() -> (VariationTextBlock:Text Object Reference), String=@n1.Mood/Status)
  ]}
 Update File Name {nodes [
   n1 = Update File Name!() -> (then, Path Parts Map:Map of Strings to Strings) {is=title K2Node_FunctionEntry}
   n2 = Get Current Animatable Creature() {is=title K2Node_CallFunction}
   n3 = Get Naming Convention(self=@n2)
   n4 = Get RemoveSpacesOnName(self=Get Current Animatable Creature())
   n5 = Get Animation Path(self=@n2)
   > Set Text 2!(
    execute
    self:EUW Animation Dialog Border Info Object Reference=Get FeedbackWidget()
    Title:Text="LOCTEXT(\"Filename\")"
    Text:Text="To Text (String)"(InString=Get Path(self=@n3, Token Map=@n1, Path Pattern=@n5, Remove Spaces On Name=@n4).File Name)
   ) -> (then) {is=title K2Node_CallFunction}
  ]}
 Update File Name from UI {nodes [
   Update File Name from UI() {is=title K2Node_FunctionEntry}
   > Update File Name!(execute, self, Path Parts Map:Map of Strings to Strings=Get Path Parts Map From UI() {is=title K2Node_CallFunction}) -> (then) {is=title K2Node_CallFunction}
  ]}
 Update Plugin Path {nodes [
   Update Plugin Path!() -> (then, Subfolder:String, Animatable Creature:Animatable Creature Object Reference) {is=title K2Node_FunctionEntry}
   > Is Valid(InputObject=Get Animatable Creature())
   n1 = Get Path Parts Map From UI()
   n2 = Get Animatable Creature()
   n3 = Get Naming Convention(self=@n2)
   n4 = Get RemoveSpacesOnName(self=Get Current Animatable Creature())
   n5 = Get Animation Path(self=@n2)
   n6 = Get FeedbackWidget()
   > Set Text 1!(
    execute
    self:EUW Animation Dialog Border Info Object Reference=@n6
    Title:Text="LOCTEXT(\"Folder\")"
    Text:Text="To Text (String)"(InString=Get Path(self=@n3, Token Map=@n1, Path Pattern=@n5, Remove Spaces On Name=@n4).Folder Path)
   ) -> (then) {is=title K2Node_CallFunction}
   > SetVisibility(self=@n6, InVisibility=Visible)
  ]}
 Update Plugin Path from UI {nodes [
   Update Plugin Path from UI() {is=title K2Node_FunctionEntry}
   n1 = Get SubfolderCustomEntryComboBox()
   > Update Plugin Path(Subfolder=GetSelectedOption(self=Get ComboBox(self=@n1)), Animatable Creature=Get Current Animatable Creature()) {is=title K2Node_CallFunction}
  ]}
 Update UI from AnimSequence {nodes [
   n1 = Update UI from AnimSequence!() -> (then, Anim Sequence Path:String, Animatable Creature:Animatable Creature Object Reference) {is=title K2Node_FunctionEntry}
   > n2 = Sequence!(execute) -> (then_0, then_1, then_2:Exec)
   n3 = AnimSequence to Name Parts!(Anim Sequence Path:String=@n1, Animatable Creature:Animatable Creature Object Reference=@n1.Animatable Creature) -> (Subfolder:String, Anim Set:String, Name:String, Descriptor:String, Direction:String, Variation:String, Success:bool) {is=title K2Node_CallFunction}
   > SetSelectedOption(self=Get ComboBox(self=Get SubfolderCustomEntryComboBox()), Option=@n3)
   > Branch(Condition=@n3.Success)
   > n4 = Sequence!(execute) -> (then_0, then_1, then_2, then_3:Exec, then_4:Exec, then_5:Exec)
   Update AnimSet ComboBox(execute=@n4.then_1)
   > SetSelectedOption(self=Get AnimSetComboBox(), Option=@n3.Anim Set)
   Set Text(execute=@n4.then_2, Text="To Text (String)"(InString=@n3.Name), self:Text Box Object Reference=Get NameTextBox!() -> (NameTextBox))
   Update Descriptor ComboBox(execute=@n4.then_3)
   > Set Value(self=Get DescriptorCustomEntryComboBox(), New Value=@n3.Descriptor)
   Update Direction ComboBox(execute=@n4.then_4)
   > SetSelectedOption(self=Get DirectionComboBox(), Option=@n3.Direction)
   Update Variation ComboBox(execute=@n4.then_5)
   > SetSelectedOption(self=Get VariationComboBox(), Option=@n3.Variation)
   Update File Name from UI(execute=@n2.then_1) {is=title K2Node_CallFunction}
   > Update Plugin Path from UI() {is=title K2Node_CallFunction}
   Validate UI Input(execute=@n2.then_2)
  ]}
 Validate Cooked Path Length {nodes [
   n1 = Validate Cooked Path Length!() -> (then, Path:String) {is=title K2Node_FunctionEntry}
   > n2 = "Execute Python Command (Advanced)"!(
    execute
    PythonCommand:String=Append!(A=MakeLiteralString(Value="object_utils.check_asset_cooked_path_limit('"), B=@n1, C:String=MakeLiteralString(Value="')")) -> (ReturnValue)
    ExecutionMode:EPythonCommandExecutionMode Enum=EvaluateStatement
    FileExecutionScope:EPythonFileExecutionScope Enum=Private
   ) -> (then, CommandResult:String, LogOutput:Array of Python Log Output Entry Structures, ReturnValue:bool)
   > Return Node!(execute, valid:bool="Equal, Case Insensitive (String)"(A=@n2, B=True))
  ]}
 Validate Path {nodes [
   Validate Path!() -> (then, InPath:String) {is=title K2Node_FunctionEntry}
   > n1 = Sequence!(execute) -> (then_0, then_1, then_2)
   n2 = ValidatePath!(InPath:String=Get InPath() -> (InPath:String)) -> (bDidSucceed:bool, OutReason:Text)
   n3 = Branch(execute=@n1.then_1, Condition=@n2)
   > n4 = Branch(Condition=Validate Cooked Path Length!(Path:String=Get InPath()) -> (valid:bool) {is=title K2Node_CallFunction})
   > n5 = DoesAssetExist(AssetPath=Get InPath())
   > Branch(Condition=@n5)
   > Return Node!(
    execute
    Valid:bool=Get Override Validation() -> (Override Validation:bool)
    Error Text:Text="LOCTEXT(\"A file with this name already exists\")"
   )
   Return Node(execute=@n4.else, Valid=false, Error Text="LOCTEXT(\"The cooked path is too long\")")
   Return Node(execute=@n3.else, Valid=false, Error Text=@n2.OutReason)
   Return Node(execute=@n1.then_2)
   Branch(execute=@n1.then_0, Condition=IsEmpty(InString=Get InPath()))
   > Return Node(Valid=false)
  ]}
 Validate UI Input {nodes [
   Validate UI Input() {is=title K2Node_FunctionEntry}
   > Set Is Valid Local(Is Valid Local:bool=true) -> (Output_Get:bool)
   > n1 = Sequence!(execute) -> (then_0, then_1, then_2, then_3:Exec, then_4:Exec, then_5:Exec, then_6:Exec, then_7:Exec, then_8:Exec)
   > n2 = Sequence!(execute) -> (then_0, then_1, then_2)
   Branch(execute=@n2.then_2, Condition=Get Is Valid Local() -> (Is Valid Local:bool))
   > Set Error 1!(execute, self:EUW Animation Dialog Border Info Object Reference=Get FeedbackWidget(), Text:Text) -> (then) {is=title K2Node_CallFunction}
   n3 = Get PluginComboBox()
   Branch(execute=@n2.then_1, Condition=IsEmpty(InString=GetSelectedOption(self=@n3)))
   > Set Error 2!(
    execute
    self:EUW Animation Dialog Border Info Object Reference=Get FeedbackWidget()
    Text:Text="LOCTEXT(\"Please Select a Plugin\")"
    Specified Color:Linear Color Structure=(R=1,G=0.1,B=0.1,A=1)
   ) -> (then) {is=title K2Node_CallFunction}
   > Set Is Valid Local(Is Valid Local=false)
   > Return Node!(execute, Is Valid:bool=false)
   n4 = Get SubfolderCustomEntryComboBox()
   n5 = Get NameTextBox!(self:EUW Custom Entry Combo Box Object Reference=@n4) -> (NameTextBox)
   n6 = Branch(execute=@n2.then_0, Condition=IsVisible(self=@n5))
   > Branch(Condition=TextIsEmpty(InText=Get Text(self=@n5)))
   > Set Error 2(self=Get FeedbackWidget(), Text="LOCTEXT(\"Please specify a Subfolder\")", Specified Color=(R=1,G=0.1,B=0.1,A=1))
   > Set Is Valid Local(Is Valid Local=false)
   > Return Node(Is Valid=false)
   n7 = Get ComboBox(self=@n4)
   Branch(execute=@n6.else, Condition=IsEmpty(InString=GetSelectedOption(self=@n7)))
   > Set Error 2(self=Get FeedbackWidget(), Text="LOCTEXT(\"Please Select a Subfolder\")", Specified Color=(R=1,G=0.1,B=0.1,A=1))
   > Set Is Valid Local(Is Valid Local=false)
   > Return Node(Is Valid=false)
   n8 = Sequence(execute=@n1.then_1)
   n9 = AND Boolean(
    A=IsEmpty(InString=Get Name())
    B=Select!(
     NewEnumerator0:bool=true
     NewEnumerator1:bool=true
     NewEnumerator2:bool=false
     Index:Enum_AC_Naming Enum=Get Naming Settings!() -> (Naming Settings=NewEnumerator0)
    ) -> (ReturnValue:bool)
   )
   > Branch(Condition=@n9)
   > Set Error 2(self=Get FeedbackWidget(), Text="LOCTEXT(\"Please set a Name\")", Specified Color=(R=1,G=0.1,B=0.1,A=1))
   > Set Is Valid Local(Is Valid Local=false)
   > Return Node(Is Valid=false)
   n10 = IsEmpty(InString=Get Asset() {is=title K2Node_CallFunction})
   Branch(execute=@n8.then_1, Condition=@n10)
   > Set Error 2(self=Get FeedbackWidget(), Text="LOCTEXT(\"Please select an asset\")", Specified Color=(R=1,G=0.1,B=0.1,A=1))
   > Set Is Valid Local(Is Valid Local=false)
   > Return Node(Is Valid=false)
   Branch(execute=@n8.then_2, Condition=AND Boolean(A=NOT Boolean(A=@n9), B=NOT Boolean(A=@n10)))
   > Set Error 2(self=Get FeedbackWidget(), Specified Color=(R=1,G=0.1,B=0.1,A=1))
   n11 = Sequence!(execute=@n1.then_7) -> (then_0, then_1, then_2, then_3:Exec, then_4:Exec)
   > n12 = Generate Anim and Level Sequence Paths() {is=title K2Node_CallFunction}
   n13 = Validate Path!(execute=@n11.then_4, self, InPath:String=@n12.Level Sequence) -> (then, Valid:bool, Error Text:Text) {is=title K2Node_CallFunction}
   > n14 = Branch(Condition=@n13)
   Set Error 2(execute=@n14.else, self=Get FeedbackWidget(), Text=@n13.Error Text, Specified Color=(R=1,G=0.1,B=0.1,A=1))
   > Return Node(Is Valid=false)
   n15 = Validate Path(execute=@n11.then_1, InPath=@n12.Animation Path)
   > n16 = Branch(Condition=@n15)
   > n17 = DoesDirectoryExist(DirectoryPath=@n12.Animation Path)
   > n18 = Branch(Condition=@n17)
   n19 = Get FeedbackWidget()
   Set Error 2(execute=@n18.else, self=@n19, Text="LOCTEXT(\"The AnimSequence directory will be created for you.\")", Specified Color=(R=1,G=0.6,B=0,A=1))
   Set Error 2(execute=@n16.else, self=@n19, Text="LOCTEXT(\"The AnimSequence path is invalid.\")", Specified Color=(R=1,G=0.1,B=0.1,A=1))
   > Return Node(Is Valid=false)
   n20 = Validate Path(execute=@n11.then_2, InPath=@n12.Source Path)
   > n21 = Branch(Condition=@n20)
   > n22 = DoesDirectoryExist(DirectoryPath=@n12.Source Path)
   > n23 = Branch(Condition=@n22)
   n24 = Get FeedbackWidget()
   Set Error 2(execute=@n23.else, self=@n24, Text="LOCTEXT(\"The Level Sequence directory will be created for you.\")", Specified Color=(R=1,G=0.6,B=0,A=1))
   > PrintString(InString=Append!(A="The following path will be created: ", B=@n12.Source Path) -> (ReturnValue))
   Set Error 2(execute=@n21.else, self=@n24, Text="LOCTEXT(\"The Level Sequence direcotry does not exist.\")", Specified Color=(R=1,G=0.1,B=0.1,A=1))
   > Return Node(Is Valid=false)
   n25 = Validate Path(execute=@n11.then_3, InPath=@n12)
   > n26 = Branch(Condition=@n25)
   Set Error 2(execute=@n26.else, self=Get FeedbackWidget(), Text=@n25.Error Text, Specified Color=(R=1,G=0.1,B=0.1,A=1))
   > Return Node(Is Valid=false)
   Set bIsEnabled!(execute=@n1.then_8, bIsEnabled:bool=true, self:Widget Object Reference=Get CreateButton() -> (CreateButton:Button Object Reference)) -> (then, Output_Get:bool)
   > Return Node(Is Valid=true)
  ]}
]
variables [
 Animatable Creatures {type="array<AnimatableCreature_C>", default []}
 CurrentUser {type=string, default=""}
 Developer Mode {type=boolean, default=true}
 Generate Animation Paths {default=()}
 Naming Settings {type=Enum_AC_Naming, default=Guided}
 Override Validation {type=boolean}
 Populated Subfolders {type="array<string>", default []}
 Project {type=BP_AnimationBrowser_Project_C, default=null}
]
widgets {
 info {rootWidgetClass=&Border, widgetCount=64}
 widgets [
  Border_0 {
   widgetClassPath=&Border
   widget=Border_0
   properties {horizontalAlignment=HAlign_Center, verticalAlignment=VAlign_Center, brushColor=<rgba 0 0 0 0.5>}
  }
  SizeBox_0 {
   widgetClassPath=&SizeBox
   parent=Border_0
   widget=SizeBox_0
   properties {bOverride_MinDesiredWidth=true, maxDesiredHeight=500, maxDesiredWidth=500, minDesiredHeight=320, minDesiredWidth=400}
   slot {class=BorderSlot, properties {horizontalAlignment=HAlign_Center, verticalAlignment=VAlign_Center}}
  }
  Border_1 {
   widgetClassPath=&Border
   parent=SizeBox_0
   widget=Border_1
   properties {brushColor=<rgba 0.0157 0.0157 0.0157 1>, padding=<ltrb 5 5 5 5>}
   slot {class=SizeBoxSlot, properties {padding=<ltrb 5 5 5 5>}}
  }
  VerticalBox_0 {widgetClassPath=&VerticalBox, parent=Border_1, widget=VerticalBox_0, slot {class=BorderSlot, properties {padding=<ltrb 5 5 5 5>}}}
  SizeBox_1 {
   widgetClassPath=&SizeBox
   parent=VerticalBox_0
   widget=SizeBox_1
   properties {bOverride_MinDesiredHeight=true, minDesiredHeight=20}
   slot {class=VerticalBoxSlot, properties {verticalAlignment=VAlign_Center}}
  }
  Border_274 {
   widgetClassPath=&Border
   parent=SizeBox_1
   widget=Border_274
   properties {verticalAlignment=VAlign_Center, brushColor=<rgba 0.0275 0.0275 0.0275 1>, padding=<ltrb 5 5 5 5>}
   slot {class=SizeBoxSlot, properties {padding=<ltrb 0 0 0 2.5>}}
  }
  VerticalBox_5 {
   widgetClassPath=&VerticalBox
   parent=Border_274
   widget=VerticalBox_5
   slot {class=BorderSlot, properties {verticalAlignment=VAlign_Center, padding=<ltrb 5 5 5 5>}}
  }
  TitleTextBlock {
   widgetClassPath=&TextBlock
   bIsVariable=true
   parent=VerticalBox_5
   widget=TitleTextBlock
   properties {
    text=Create New Animation
    colorAndOpacity {colorUseRule=UseColor_Specified, specifiedColor=<rgba 0.5255 0.5255 0.5255 1>}
    font {
     bForceMonospaced=false
     bMaterialIsStencil=false
     fontMaterial=null
     fontObject=&Roboto
     letterSpacing=0
     monospacedWidth=1
     size=11.25
     skewAmount=0
     typefaceFontName=Bold
     outlineSettings {
      bApplyOutlineToDropShadows=false
      bMiteredCorners=false
      bSeparateFillAlpha=false
      outlineMaterial=null
      outlineSize=0
      outlineColor=<rgba 0 0 0 1>
     }
    }
   }
   slot {class=VerticalBoxSlot}
  }
  Description {
   widgetClassPath=&TextBlock
   bIsVariable=true
   parent=VerticalBox_5
   widget=Description
   properties {
    autoWrapText=true
    text=Description
    visibility=Collapsed
    colorAndOpacity {colorUseRule=UseColor_Specified, specifiedColor=<rgba 0.5255 0.5255 0.5255 1>}
    font {
     bForceMonospaced=false
     bMaterialIsStencil=false
     fontMaterial=null
     fontObject=&Roboto
     letterSpacing=0
     monospacedWidth=1
     size=9
     skewAmount=0
     typefaceFontName=Regular
     outlineSettings {
      bApplyOutlineToDropShadows=false
      bMiteredCorners=false
      bSeparateFillAlpha=false
      outlineMaterial=null
      outlineSize=0
      outlineColor=<rgba 0 0 0 1>
     }
    }
   }
   slot {class=VerticalBoxSlot, properties {padding=<ltrb 4 2 4 2>}}
  }
  HeaderWidget {
   widgetClassPath=&/EUW_AnimationDialog_BorderInfo
   bIsVariable=true
   parent=VerticalBox_0
   widget=HeaderWidget
   properties {visibility=Collapsed}
   slot {class=VerticalBoxSlot, properties {padding=<ltrb 0 2.5 0 2.5>}}
  }
  EditorUtilityScrollBox_0 {
   widgetClassPath=&EditorUtilityScrollBox
   parent=VerticalBox_0
   widget=EditorUtilityScrollBox_0
   slot {class=VerticalBoxSlot, properties {size {sizeRule=Fill, value=1}}}
  }
  Border_2 {
   widgetClassPath=&Border
   parent=EditorUtilityScrollBox_0
   widget=Border_2
   properties {brushColor=<rgba 0.0273 0.0273 0.0273 1>}
   slot {class=ScrollBoxSlot, properties {padding=<ltrb 0 2.5 0 5>}}
  }
  VerticalBox_4 {widgetClassPath=&VerticalBox, parent=Border_2, widget=VerticalBox_4, slot {class=BorderSlot}}
  PluginTextBlock_2 {
   widgetClassPath=&TextBlock
   parent=VerticalBox_4
   widget=PluginTextBlock_2
   properties {
    text=Asset
    toolTipText=What do you want to import and where do you want it to go?
    colorAndOpacity {colorUseRule=UseColor_Specified, specifiedColor=<rgba 0.5255 0.5255 0.5255 1>}
    font {
     bForceMonospaced=false
     bMaterialIsStencil=false
     fontMaterial=null
     fontObject=&Roboto
     letterSpacing=0
     monospacedWidth=1
     size=9
     skewAmount=0
     typefaceFontName=Bold
     outlineSettings {
      bApplyOutlineToDropShadows=false
      bMiteredCorners=false
      bSeparateFillAlpha=false
      outlineMaterial=null
      outlineSize=0
      outlineColor=<rgba 0 0 0 1>
     }
    }
   }
   slot {class=VerticalBoxSlot, properties {horizontalAlignment=HAlign_Left, padding=<ltrb 0 2 0 0>}}
  }
  ProjectHorizontalBox {
   widgetClassPath=&HorizontalBox
   bIsVariable=true
   parent=VerticalBox_4
   widget=ProjectHorizontalBox
   properties {toolTipText=Which asset do you want animate?, visibility=Collapsed}
   slot {class=VerticalBoxSlot, properties {padding=<ltrb 0 1 0 1>}}
  }
  SizeBox_9 {
   widgetClassPath=&SizeBox
   parent=ProjectHorizontalBox
   widget=SizeBox_9
   properties {bOverride_MinDesiredWidth=true, minDesiredWidth=100}
   slot {class=HorizontalBoxSlot, properties {verticalAlignment=VAlign_Center}}
  }
  ProjectTextBlock {
   widgetClassPath=&TextBlock
   bIsVariable=true
   parent=SizeBox_9
   widget=ProjectTextBlock
   properties {
    text=Project
    colorAndOpacity {colorUseRule=UseColor_Specified, specifiedColor=<rgba 0.5255 0.5255 0.5255 1>}
    font {
     bForceMonospaced=false
     bMaterialIsStencil=false
     fontMaterial=null
     fontObject=&Roboto
     letterSpacing=0
     monospacedWidth=1
     size=10
     skewAmount=0
     typefaceFontName=Bold
     outlineSettings {
      bApplyOutlineToDropShadows=false
      bMiteredCorners=false
      bSeparateFillAlpha=false
      outlineMaterial=null
      outlineSize=0
      outlineColor=<rgba 0 0 0 1>
     }
    }
   }
   slot {class=SizeBoxSlot, properties {horizontalAlignment=HAlign_Right}}
  }
  ProjectComboBox {
   widgetClassPath=/Script/FortniteEditor.SearchableStringComboBox
   bIsVariable=true
   parent=ProjectHorizontalBox
   widget=ProjectComboBox
   properties {
    font {
     bForceMonospaced=false
     bMaterialIsStencil=false
     fontMaterial=null
     fontObject=&Roboto
     letterSpacing=0
     monospacedWidth=1
     size=10
     skewAmount=0
     typefaceFontName=Regular
     outlineSettings {
      bApplyOutlineToDropShadows=false
      bMiteredCorners=false
      bSeparateFillAlpha=false
      outlineMaterial=null
      outlineSize=0
      outlineColor=<rgba 0 0 0 1>
     }
    }
    foregroundColor {colorUseRule=UseColor_Specified, specifiedColor=<rgba 0.5271 0.5271 0.5271 1>}
    itemStyle {
     bUseParentRowBrush=false
     activeBrush {
      bIsDynamicallyLoaded=false
      drawAs=Image
      imageType=NoImage
      mirroring=NoMirror
      resourceName=None
      resourceObject=null
      tiling=NoTile
      imageSize=<0 0>
      margin=<ltrb 0 0 0 0>
      outlineSettings {
       bUseBrushTransparency=false
       roundingType=HalfHeightRadius
       width=0
       color {colorUseRule=UseColor_Specified, specifiedColor=<rgba 0 0 0 0>}
       cornerRadii=<0 0 0 0>
      }
      tintColor {colorUseRule=UseColor_Specified, specifiedColor=<rgba 0 0.162 0.7454 1>}
      uVRegion {bIsValid=false, max=<0 0>, min=<0 0>}
     }
     activeHighlightedBrush {
      bIsDynamicallyLoaded=false
      drawAs=Image
      imageType=NoImage
      mirroring=NoMirror
      resourceName=None
      resourceObject=null
      tiling=NoTile
      imageSize=<0 0>
      margin=<ltrb 0 0 0 0>
      outlineSettings {
       bUseBrushTransparency=false
       roundingType=HalfHeightRadius
       width=0
       color {colorUseRule=UseColor_Specified, specifiedColor=<rgba 0 0 0 0>}
       cornerRadii=<0 0 0 0>
      }
      tintColor {colorUseRule=UseColor_Specified, specifiedColor=<rgba 0.0252 0.0319 0.0423 1>}
      uVRegion {bIsValid=false, max=<0 0>, min=<0 0>}
     }
     activeHoveredBrush {
      bIsDynamicallyLoaded=false
      drawAs=Image
      imageType=NoImage
      mirroring=NoMirror
      resourceName=None
      resourceObject=null
      tiling=NoTile
      imageSize=<0 0>
      margin=<ltrb 0 0 0 0>
      outlineSettings {
       bUseBrushTransparency=false
       roundingType=HalfHeightRadius
       width=0
       color {colorUseRule=UseColor_Specified, specifiedColor=<rgba 0 0 0 0>}
       cornerRadii=<0 0 0 0>
      }
      tintColor {colorUseRule=UseColor_Specified, specifiedColor=<rgba 0 0.162 0.7454 1>}
      uVRegion {bIsValid=false, max=<0 0>, min=<0 0>}
     }
     dropIndicator_Above {
      bIsDynamicallyLoaded=false
      drawAs=Box
      imageType=FullColor
      mirroring=NoMirror
      resourceName=../../../Engine/Content/Slate/Common/DropZoneIndicator_Above.png
      resourceObject=null
      tiling=NoTile
      imageSize=<0 0>
      margin=<ltrb 0.625 0.625 0 0>
      outlineSettings {
       bUseBrushTransparency=false
       roundingType=HalfHeightRadius
       width=0
       color {colorUseRule=UseColor_Specified, specifiedColor=<rgba 0 0 0 0>}
       cornerRadii=<0 0 0 0>
      }
      tintColor {colorUseRule=UseColor_Specified, specifiedColor=<rgba 0.0091 0.3515 0.9647 1>}
      uVRegion {bIsValid=false, max=<0 0>, min=<0 0>}
     }
     dropIndicator_Below {
      bIsDynamicallyLoaded=false
      drawAs=Box
      imageType=FullColor
      mirroring=NoMirror
      resourceName=../../../Engine/Content/Slate/Common/DropZoneIndicator_Below.png
      resourceObject=null
      tiling=NoTile
      imageSize=<0 0>
      margin=<ltrb 0.625 0 0 0.625>
      outlineSettings {
       bUseBrushTransparency=false
       roundingType=HalfHeightRadius
       width=0
       color {colorUseRule=UseColor_Specified, specifiedColor=<rgba 0 0 0 0>}
       cornerRadii=<0 0 0 0>
      }
      tintColor {colorUseRule=UseColor_Specified, specifiedColor=<rgba 0.0091 0.3515 0.9647 1>}
      uVRegion {bIsValid=false, max=<0 0>, min=<0 0>}
     }
     dropIndicator_Onto {
      bIsDynamicallyLoaded=false
      drawAs=Box
      imageType=FullColor
      mirroring=NoMirror
      resourceName=../../../Engine/Content/Slate/Common/DropZoneIndicator_Onto.png
      resourceObject=null
      tiling=NoTile
      imageSize=<0 0>
      margin=<ltrb 0.25 0.25 0.25 0.25>
      outlineSettings {
       bUseBrushTransparency=false
       roundingType=HalfHeightRadius
       width=0
       color {colorUseRule=UseColor_Specified, specifiedColor=<rgba 0 0 0 0>}
       cornerRadii=<0 0 0 0>
      }
      tintColor {colorUseRule=UseColor_Specified, specifiedColor=<rgba 0.0091 0.3515 0.9647 1>}
      uVRegion {bIsValid=false, max=<0 0>, min=<0 0>}
     }
     evenRowBackgroundBrush {
      bIsDynamicallyLoaded=false
      drawAs=Image
      imageType=NoImage
      mirroring=NoMirror
      resourceName=None
      resourceObject=null
      tiling=NoTile
      imageSize=<0 0>
      margin=<ltrb 0 0 0 0>
      outlineSettings {
       bUseBrushTransparency=false
       roundingType=HalfHeightRadius
       width=0
       color {colorUseRule=UseColor_Specified, specifiedColor=<rgba 0 0 0 0>}
       cornerRadii=<0 0 0 0>
      }
      tintColor {colorUseRule=UseColor_Specified, specifiedColor=<rgba 0.0103 0.0103 0.0103 1>}
      uVRegion {bIsValid=false, max=<0 0>, min=<0 0>}
     }
     evenRowBackgroundHoveredBrush {
      bIsDynamicallyLoaded=false
      drawAs=Image
      imageType=NoImage
      mirroring=NoMirror
      resourceName=None
      resourceObject=null
      tiling=NoTile
      imageSize=<0 0>
      margin=<ltrb 0 0 0 0>
      outlineSettings {
       bUseBrushTransparency=false
       roundingType=HalfHeightRadius
       width=0
       color {colorUseRule=UseColor_Specified, specifiedColor=<rgba 0 0 0 0>}
       cornerRadii=<0 0 0 0>
      }
      tintColor {colorUseRule=UseColor_Specified, specifiedColor=<rgba 0.0176 0.0176 0.0176 1>}
      uVRegion {bIsValid=false, max=<0 0>, min=<0 0>}
     }
     inactiveBrush {
      bIsDynamicallyLoaded=false
      drawAs=Image
      imageType=NoImage
      mirroring=NoMirror
      resourceName=None
      resourceObject=null
      tiling=NoTile
      imageSize=<0 0>
      margin=<ltrb 0 0 0 0>
      outlineSettings {
       bUseBrushTransparency=false
       roundingType=HalfHeightRadius
       width=0
       color {colorUseRule=UseColor_Specified, specifiedColor=<rgba 0 0 0 0>}
       cornerRadii=<0 0 0 0>
      }
      tintColor {colorUseRule=UseColor_Specified, specifiedColor=<rgba 0.0513 0.0953 0.159 1>}
      uVRegion {bIsValid=false, max=<0 0>, min=<0 0>}
     }
     inactiveHighlightedBrush {
      bIsDynamicallyLoaded=false
      drawAs=Image
      imageType=NoImage
      mirroring=NoMirror
      resourceName=None
      resourceObject=null
      tiling=NoTile
      imageSize=<0 0>
      margin=<ltrb 0 0 0 0>
      outlineSettings {
       bUseBrushTransparency=false
       roundingType=HalfHeightRadius
       width=0
       color {colorUseRule=UseColor_Specified, specifiedColor=<rgba 0 0 0 0>}
       cornerRadii=<0 0 0 0>
      }
      tintColor {colorUseRule=UseColor_Specified, specifiedColor=<rgba 0.0252 0.0319 0.0423 1>}
      uVRegion {bIsValid=false, max=<0 0>, min=<0 0>}
     }
     inactiveHoveredBrush {
      bIsDynamicallyLoaded=false
      drawAs=Image
      imageType=NoImage
      mirroring=NoMirror
      resourceName=None
      resourceObject=null
      tiling=NoTile
      imageSize=<0 0>
      margin=<ltrb 0 0 0 0>
      outlineSettings {
       bUseBrushTransparency=false
       roundingType=HalfHeightRadius
       width=0
       color {colorUseRule=UseColor_Specified, specifiedColor=<rgba 0 0 0 0>}
       cornerRadii=<0 0 0 0>
      }
      tintColor {colorUseRule=UseColor_Specified, specifiedColor=<rgba 0.0513 0.0953 0.159 1>}
      uVRegion {bIsValid=false, max=<0 0>, min=<0 0>}
     }
     oddRowBackgroundBrush {
      bIsDynamicallyLoaded=false
      drawAs=Image
      imageType=NoImage
      mirroring=NoMirror
      resourceName=None
      resourceObject=null
      tiling=NoTile
      imageSize=<0 0>
      margin=<ltrb 0 0 0 0>
      outlineSettings {
       bUseBrushTransparency=false
       roundingType=HalfHeightRadius
       width=0
       color {colorUseRule=UseColor_Specified, specifiedColor=<rgba 0 0 0 0>}
       cornerRadii=<0 0 0 0>
      }
      tintColor {colorUseRule=UseColor_Specified, specifiedColor=<rgba 0.0103 0.0103 0.0103 1>}
      uVRegion {bIsValid=false, max=<0 0>, min=<0 0>}
     }
     oddRowBackgroundHoveredBrush {
      bIsDynamicallyLoaded=false
      drawAs=Image
      imageType=NoImage
      mirroring=NoMirror
      resourceName=None
      resourceObject=null
      tiling=NoTile
      imageSize=<0 0>
      margin=<ltrb 0 0 0 0>
      outlineSettings {
       bUseBrushTransparency=false
       roundingType=HalfHeightRadius
       width=0
       color {colorUseRule=UseColor_Specified, specifiedColor=<rgba 0 0 0 0>}
       cornerRadii=<0 0 0 0>
      }
      tintColor {colorUseRule=UseColor_Specified, specifiedColor=<rgba 0.0176 0.0176 0.0176 1>}
      uVRegion {bIsValid=false, max=<0 0>, min=<0 0>}
     }
     parentRowBackgroundBrush {
      bIsDynamicallyLoaded=false
      drawAs=Image
      imageType=NoImage
      mirroring=NoMirror
      resourceName=None
      resourceObject=null
      tiling=NoTile
      imageSize=<32 32>
      margin=<ltrb 0 0 0 0>
      outlineSettings {
       bUseBrushTransparency=false
       roundingType=HalfHeightRadius
       width=0
       color {colorUseRule=UseColor_Specified, specifiedColor=<rgba 0 0 0 0>}
       cornerRadii=<0 0 0 0>
      }
      tintColor {colorUseRule=UseColor_Specified, specifiedColor=<rgba 1 1 1 1>}
      uVRegion {bIsValid=false, max=<0 0>, min=<0 0>}
     }
     parentRowBackgroundHoveredBrush {
      bIsDynamicallyLoaded=false
      drawAs=Image
      imageType=NoImage
      mirroring=NoMirror
      resourceName=None
      resourceObject=null
      tiling=NoTile
      imageSize=<32 32>
      margin=<ltrb 0 0 0 0>
      outlineSettings {
       bUseBrushTransparency=false
       roundingType=HalfHeightRadius
       width=0
       color {colorUseRule=UseColor_Specified, specifiedColor=<rgba 0 0 0 0>}
       cornerRadii=<0 0 0 0>
      }
      tintColor {colorUseRule=UseColor_Specified, specifiedColor=<rgba 1 1 1 1>}
      uVRegion {bIsValid=false, max=<0 0>, min=<0 0>}
     }
     selectedTextColor {colorUseRule=UseColor_Specified, specifiedColor=<rgba 1 1 1 1>}
     selectorFocusedBrush {
      bIsDynamicallyLoaded=false
      drawAs=Border
      imageType=FullColor
      mirroring=NoMirror
      resourceName=../../../Engine/Content/Slate/Common/Selector.png
      resourceObject=null
      tiling=Both
      imageSize=<0 0>
      margin=<ltrb 0.25 0.25 0.25 0.25>
      outlineSettings {
       bUseBrushTransparency=false
       roundingType=HalfHeightRadius
       width=0
       color {colorUseRule=UseColor_Specified, specifiedColor=<rgba 0 0 0 0>}
       cornerRadii=<0 0 0 0>
      }
      tintColor {colorUseRule=UseColor_Specified, specifiedColor=<rgba 0.701 0.225 0.003 1>}
      uVRegion {bIsValid=false, max=<0 0>, min=<0 0>}
     }
     textColor {colorUseRule=UseColor_Specified, specifiedColor=<rgba 0.5271 0.5271 0.5271 1>}
    }
    widgetStyle {
     comboButtonStyle {
      downArrowAlign=VAlign_Center
      buttonStyle {
       clickedSlateSound {resourceObject=null}
       disabled {
        bIsDynamicallyLoaded=false
        drawAs=RoundedBox
        imageType=NoImage
        mirroring=NoMirror
        resourceName=None
        resourceObject=null
        tiling=NoTile
        imageSize=<0 0>
        margin=<ltrb 0 0 0 0>
        outlineSettings {
         bUseBrushTransparency=false
         roundingType=FixedRadius
         width=1
         color {colorUseRule=UseColor_ColorTable, specifiedColor=<rgba 0 0 0 0>}
         cornerRadii=<4 4 4 4>
        }
        tintColor {colorUseRule=UseColor_Specified, specifiedColor=<rgba 0.0395 0.0395 0.0395 1>}
        uVRegion {bIsValid=false, max=<0 0>, min=<0 0>}
       }
       disabledForeground {colorUseRule=UseColor_ColorTable, specifiedColor=<rgba 0 0 0 0>}
       hovered {
        bIsDynamicallyLoaded=false
        drawAs=RoundedBox
        imageType=NoImage
        mirroring=NoMirror
        resourceName=None
        resourceObject=null
        tiling=NoTile
        imageSize=<0 0>
        margin=<ltrb 0 0 0 0>
        outlineSettings {
         bUseBrushTransparency=false
         roundingType=FixedRadius
         width=1
         color {colorUseRule=UseColor_ColorTable, specifiedColor=<rgba 0 0 0 0>}
         cornerRadii=<4 4 4 4>
        }
        tintColor {colorUseRule=UseColor_Specified, specifiedColor=<rgba 0.0048 0.0048 0.0048 1>}
        uVRegion {bIsValid=false, max=<0 0>, min=<0 0>}
       }
       hoveredForeground {colorUseRule=UseColor_ColorTable, specifiedColor=<rgba 0 0 0 0>}
       hoveredSlateSound {resourceObject=null}
       normal {
        bIsDynamicallyLoaded=false
        drawAs=RoundedBox
        imageType=NoImage
        mirroring=NoMirror
        resourceName=None
        resourceObject=null
        tiling=NoTile
        imageSize=<0 0>
        margin=<ltrb 0 0 0 0>
        outlineSettings {
         bUseBrushTransparency=false
         roundingType=FixedRadius
         width=1
         color {colorUseRule=UseColor_ColorTable, specifiedColor=<rgba 0 0 0 0>}
         cornerRadii=<4 4 4 4>
        }
        tintColor {colorUseRule=UseColor_Specified, specifiedColor=<rgba 0.0048 0.0048 0.0048 1>}
        uVRegion {bIsValid=false, max=<0 0>, min=<0 0>}
       }
       normalForeground {colorUseRule=UseColor_ColorTable, specifiedColor=<rgba 0 0 0 0>}
       normalPadding=<ltrb 8 1 8 1>
       pressed {
        bIsDynamicallyLoaded=false
        drawAs=RoundedBox
        imageType=NoImage
        mirroring=NoMirror
        resourceName=None
        resourceObject=null
        tiling=NoTile
        imageSize=<0 0>
        margin=<ltrb 0 0 0 0>
        outlineSettings {
         bUseBrushTransparency=false
         roundingType=FixedRadius
         width=1
         color {colorUseRule=UseColor_ColorTable, specifiedColor=<rgba 0 0 0 0>}
         cornerRadii=<4 4 4 4>
        }
        tintColor {colorUseRule=UseColor_Specified, specifiedColor=<rgba 0.0048 0.0048 0.0048 1>}
        uVRegion {bIsValid=false, max=<0 0>, min=<0 0>}
       }
       pressedForeground {colorUseRule=UseColor_ColorTable, specifiedColor=<rgba 0 0 0 0>}
       pressedPadding=<ltrb 8 2 8 0>
       pressedSlateSound {resourceObject=null}
      }
      contentPadding=<ltrb 0 0 0 0>
      downArrowImage {
       bIsDynamicallyLoaded=false
       drawAs=Image
       imageType=Vector
       mirroring=NoMirror
       resourceName=../../../Engine/Content/Slate/Starship/CoreWidgets/ComboBox/wide-chevron-down.svg
       resourceObject=null
       tiling=NoTile
       imageSize=<16 16>
       margin=<ltrb 0 0 0 0>
       outlineSettings {
        bUseBrushTransparency=false
        roundingType=HalfHeightRadius
        width=0
        color {colorUseRule=UseColor_Specified, specifiedColor=<rgba 0 0 0 0>}
        cornerRadii=<0 0 0 0>
       }
       tintColor {colorUseRule=UseColor_Specified, specifiedColor=<rgba 1 1 1 1>}
       uVRegion {bIsValid=false, max=<0 0>, min=<0 0>}
      }
      downArrowPadding=<ltrb 2 2 2 2>
      menuBorderBrush {
       bIsDynamicallyLoaded=false
       drawAs=RoundedBox
       imageType=NoImage
       mirroring=NoMirror
       resourceName=None
       resourceObject=null
       tiling=NoTile
       imageSize=<0 0>
       margin=<ltrb 0 0 0 0>
       outlineSettings {
        bUseBrushTransparency=false
        roundingType=FixedRadius
        width=1
        color {colorUseRule=UseColor_ColorTable, specifiedColor=<rgba 0 0 0 0>}
        cornerRadii=<0 0 0 0>
       }
       tintColor {colorUseRule=UseColor_Specified, specifiedColor=<rgba 0.0395 0.0395 0.0395 1>}
       uVRegion {bIsValid=false, max=<0 0>, min=<0 0>}
      }
      menuBorderPadding=<ltrb 0 0 0 0>
      shadowColorAndOpacity=<rgba 0 0 0 1>
      shadowOffset=<0 0>
     }
     contentPadding=<ltrb 0 0 0 0>
     menuRowPadding=<ltrb 8 3 8 3>
     pressedSlateSound {resourceObject=null}
     selectionChangeSlateSound {resourceObject=null}
    }
   }
   slot {class=HorizontalBoxSlot, properties {padding=<ltrb 5 0 0 0>, size {sizeRule=Fill, value=1}}}
  }
  AssetHorizontalBox {
   widgetClassPath=&HorizontalBox
   bIsVariable=true
   parent=VerticalBox_4
   widget=AssetHorizontalBox
   properties {toolTipText=Which asset do you want animate?}
   slot {class=VerticalBoxSlot, properties {padding=<ltrb 0 1 0 1>}}
  }
  SizeBox_6 {
   widgetClassPath=&SizeBox
   parent=AssetHorizontalBox
   widget=SizeBox_6
   properties {bOverride_MinDesiredWidth=true, minDesiredWidth=100}
   slot {class=HorizontalBoxSlot, properties {verticalAlignment=VAlign_Center}}
  }
  AssetTextBlock {
   widgetClassPath=&TextBlock
   bIsVariable=true
   parent=SizeBox_6
   widget=AssetTextBlock
   properties {
    text=Asset
    colorAndOpacity {colorUseRule=UseColor_Specified, specifiedColor=<rgba 0.5255 0.5255 0.5255 1>}
    font {
     bForceMonospaced=false
     bMaterialIsStencil=false
     fontMaterial=null
     fontObject=&Roboto
     letterSpacing=0
     monospacedWidth=1
     size=10
     skewAmount=0
     typefaceFontName=Bold
     outlineSettings {
      bApplyOutlineToDropShadows=false
      bMiteredCorners=false
      bSeparateFillAlpha=false
      outlineMaterial=null
      outlineSize=0
      outlineColor=<rgba 0 0 0 1>
     }
    }
   }
   slot {class=SizeBoxSlot, properties {horizontalAlignment=HAlign_Right}}
  }
  AssetComboBox {
   widgetClassPath=&SearchableStringComboBox
   bIsVariable=true
   parent=AssetHorizontalBox
   widget=AssetComboBox
   properties {
    font {
     bForceMonospaced=false
     bMaterialIsStencil=false
     fontMaterial=null
     fontObject=&Roboto
     letterSpacing=0
     monospacedWidth=1
     size=10
     skewAmount=0
     typefaceFontName=Regular
     outlineSettings {
      bApplyOutlineToDropShadows=false
      bMiteredCorners=false
      bSeparateFillAlpha=false
      outlineMaterial=null
      outlineSize=0
      outlineColor=<rgba 0 0 0 1>
     }
    }
    foregroundColor {colorUseRule=UseColor_Specified, specifiedColor=<rgba 0.5271 0.5271 0.5271 1>}
    itemStyle {
     bUseParentRowBrush=false
     activeBrush {
      bIsDynamicallyLoaded=false
      drawAs=Image
      imageType=NoImage
      mirroring=NoMirror
      resourceName=None
      resourceObject=null
      tiling=NoTile
      imageSize=<0 0>
      margin=<ltrb 0 0 0 0>
      outlineSettings {
       bUseBrushTransparency=false
       roundingType=HalfHeightRadius
       width=0
       color {colorUseRule=UseColor_Specified, specifiedColor=<rgba 0 0 0 0>}
       cornerRadii=<0 0 0 0>
      }
      tintColor {colorUseRule=UseColor_Specified, specifiedColor=<rgba 0 0.162 0.7454 1>}
      uVRegion {bIsValid=false, max=<0 0>, min=<0 0>}
     }
     activeHighlightedBrush {
      bIsDynamicallyLoaded=false
      drawAs=Image
      imageType=NoImage
      mirroring=NoMirror
      resourceName=None
      resourceObject=null
      tiling=NoTile
      imageSize=<0 0>
      margin=<ltrb 0 0 0 0>
      outlineSettings {
       bUseBrushTransparency=false
       roundingType=HalfHeightRadius
       width=0
       color {colorUseRule=UseColor_Specified, specifiedColor=<rgba 0 0 0 0>}
       cornerRadii=<0 0 0 0>
      }
      tintColor {colorUseRule=UseColor_Specified, specifiedColor=<rgba 0.0252 0.0319 0.0423 1>}
      uVRegion {bIsValid=false, max=<0 0>, min=<0 0>}
     }
     activeHoveredBrush {
      bIsDynamicallyLoaded=false
      drawAs=Image
      imageType=NoImage
      mirroring=NoMirror
      resourceName=None
      resourceObject=null
      tiling=NoTile
      imageSize=<0 0>
      margin=<ltrb 0 0 0 0>
      outlineSettings {
       bUseBrushTransparency=false
       roundingType=HalfHeightRadius
       width=0
       color {colorUseRule=UseColor_Specified, specifiedColor=<rgba 0 0 0 0>}
       cornerRadii=<0 0 0 0>
      }
      tintColor {colorUseRule=UseColor_Specified, specifiedColor=<rgba 0 0.162 0.7454 1>}
      uVRegion {bIsValid=false, max=<0 0>, min=<0 0>}
     }
     dropIndicator_Above {
      bIsDynamicallyLoaded=false
      drawAs=Box
      imageType=FullColor
      mirroring=NoMirror
      resourceName=../../../Engine/Content/Slate/Common/DropZoneIndicator_Above.png
      resourceObject=null
      tiling=NoTile
      imageSize=<0 0>
      margin=<ltrb 0.625 0.625 0 0>
      outlineSettings {
       bUseBrushTransparency=false
       roundingType=HalfHeightRadius
       width=0
       color {colorUseRule=UseColor_Specified, specifiedColor=<rgba 0 0 0 0>}
       cornerRadii=<0 0 0 0>
      }
      tintColor {colorUseRule=UseColor_Specified, specifiedColor=<rgba 0.0091 0.3515 0.9647 1>}
      uVRegion {bIsValid=false, max=<0 0>, min=<0 0>}
     }
     dropIndicator_Below {
      bIsDynamicallyLoaded=false
      drawAs=Box
      imageType=FullColor
      mirroring=NoMirror
      resourceName=../../../Engine/Content/Slate/Common/DropZoneIndicator_Below.png
      resourceObject=null
      tiling=NoTile
      imageSize=<0 0>
      margin=<ltrb 0.625 0 0 0.625>
      outlineSettings {
       bUseBrushTransparency=false
       roundingType=HalfHeightRadius
       width=0
       color {colorUseRule=UseColor_Specified, specifiedColor=<rgba 0 0 0 0>}
       cornerRadii=<0 0 0 0>
      }
      tintColor {colorUseRule=UseColor_Specified, specifiedColor=<rgba 0.0091 0.3515 0.9647 1>}
      uVRegion {bIsValid=false, max=<0 0>, min=<0 0>}
     }
     dropIndicator_Onto {
      bIsDynamicallyLoaded=false
      drawAs=Box
      imageType=FullColor
      mirroring=NoMirror
      resourceName=../../../Engine/Content/Slate/Common/DropZoneIndicator_Onto.png
      resourceObject=null
      tiling=NoTile
      imageSize=<0 0>
      margin=<ltrb 0.25 0.25 0.25 0.25>
      outlineSettings {
       bUseBrushTransparency=false
       roundingType=HalfHeightRadius
       width=0
       color {colorUseRule=UseColor_Specified, specifiedColor=<rgba 0 0 0 0>}
       cornerRadii=<0 0 0 0>
      }
      tintColor {colorUseRule=UseColor_Specified, specifiedColor=<rgba 0.0091 0.3515 0.9647 1>}
      uVRegion {bIsValid=false, max=<0 0>, min=<0 0>}
     }
     evenRowBackgroundBrush {
      bIsDynamicallyLoaded=false
      drawAs=Image
      imageType=NoImage
      mirroring=NoMirror
      resourceName=None
      resourceObject=null
      tiling=NoTile
      imageSize=<0 0>
      margin=<ltrb 0 0 0 0>
      outlineSettings {
       bUseBrushTransparency=false
       roundingType=HalfHeightRadius
       width=0
       color {colorUseRule=UseColor_Specified, specifiedColor=<rgba 0 0 0 0>}
       cornerRadii=<0 0 0 0>
      }
      tintColor {colorUseRule=UseColor_Specified, specifiedColor=<rgba 0.0103 0.0103 0.0103 1>}
      uVRegion {bIsValid=false, max=<0 0>, min=<0 0>}
     }
     evenRowBackgroundHoveredBrush {
      bIsDynamicallyLoaded=false
      drawAs=Image
      imageType=NoImage
      mirroring=NoMirror
      resourceName=None
      resourceObject=null
      tiling=NoTile
      imageSize=<0 0>
      margin=<ltrb 0 0 0 0>
      outlineSettings {
       bUseBrushTransparency=false
       roundingType=HalfHeightRadius
       width=0
       color {colorUseRule=UseColor_Specified, specifiedColor=<rgba 0 0 0 0>}
       cornerRadii=<0 0 0 0>
      }
      tintColor {colorUseRule=UseColor_Specified, specifiedColor=<rgba 0.0176 0.0176 0.0176 1>}
      uVRegion {bIsValid=false, max=<0 0>, min=<0 0>}
     }
     inactiveBrush {
      bIsDynamicallyLoaded=false
      drawAs=Image
      imageType=NoImage
      mirroring=NoMirror
      resourceName=None
      resourceObject=null
      tiling=NoTile
      imageSize=<0 0>
      margin=<ltrb 0 0 0 0>
      outlineSettings {
       bUseBrushTransparency=false
       roundingType=HalfHeightRadius
       width=0
       color {colorUseRule=UseColor_Specified, specifiedColor=<rgba 0 0 0 0>}
       cornerRadii=<0 0 0 0>
      }
      tintColor {colorUseRule=UseColor_Specified, specifiedColor=<rgba 0.0513 0.0953 0.159 1>}
      uVRegion {bIsValid=false, max=<0 0>, min=<0 0>}
     }
     inactiveHighlightedBrush {
      bIsDynamicallyLoaded=false
      drawAs=Image
      imageType=NoImage
      mirroring=NoMirror
      resourceName=None
      resourceObject=null
      tiling=NoTile
      imageSize=<0 0>
      margin=<ltrb 0 0 0 0>
      outlineSettings {
       bUseBrushTransparency=false
       roundingType=HalfHeightRadius
       width=0
       color {colorUseRule=UseColor_Specified, specifiedColor=<rgba 0 0 0 0>}
       cornerRadii=<0 0 0 0>
      }
      tintColor {colorUseRule=UseColor_Specified, specifiedColor=<rgba 0.0252 0.0319 0.0423 1>}
      uVRegion {bIsValid=false, max=<0 0>, min=<0 0>}
     }
     inactiveHoveredBrush {
      bIsDynamicallyLoaded=false
      drawAs=Image
      imageType=NoImage
      mirroring=NoMirror
      resourceName=None
      resourceObject=null
      tiling=NoTile
      imageSize=<0 0>
      margin=<ltrb 0 0 0 0>
      outlineSettings {
       bUseBrushTransparency=false
       roundingType=HalfHeightRadius
       width=0
       color {colorUseRule=UseColor_Specified, specifiedColor=<rgba 0 0 0 0>}
       cornerRadii=<0 0 0 0>
      }
      tintColor {colorUseRule=UseColor_Specified, specifiedColor=<rgba 0.0513 0.0953 0.159 1>}
      uVRegion {bIsValid=false, max=<0 0>, min=<0 0>}
     }
     oddRowBackgroundBrush {
      bIsDynamicallyLoaded=false
      drawAs=Image
      imageType=NoImage
      mirroring=NoMirror
      resourceName=None
      resourceObject=null
      tiling=NoTile
      imageSize=<0 0>
      margin=<ltrb 0 0 0 0>
      outlineSettings {
       bUseBrushTransparency=false
       roundingType=HalfHeightRadius
       width=0
       color {colorUseRule=UseColor_Specified, specifiedColor=<rgba 0 0 0 0>}
       cornerRadii=<0 0 0 0>
      }
      tintColor {colorUseRule=UseColor_Specified, specifiedColor=<rgba 0.0103 0.0103 0.0103 1>}
      uVRegion {bIsValid=false, max=<0 0>, min=<0 0>}
     }
     oddRowBackgroundHoveredBrush {
      bIsDynamicallyLoaded=false
      drawAs=Image
      imageType=NoImage
      mirroring=NoMirror
      resourceName=None
      resourceObject=null
      tiling=NoTile
      imageSize=<0 0>
      margin=<ltrb 0 0 0 0>
      outlineSettings {
       bUseBrushTransparency=false
       roundingType=HalfHeightRadius
       width=0
       color {colorUseRule=UseColor_Specified, specifiedColor=<rgba 0 0 0 0>}
       cornerRadii=<0 0 0 0>
      }
      tintColor {colorUseRule=UseColor_Specified, specifiedColor=<rgba 0.0176 0.0176 0.0176 1>}
      uVRegion {bIsValid=false, max=<0 0>, min=<0 0>}
     }
     parentRowBackgroundBrush {
      bIsDynamicallyLoaded=false
      drawAs=Image
      imageType=NoImage
      mirroring=NoMirror
      resourceName=None
      resourceObject=null
      tiling=NoTile
      imageSize=<32 32>
      margin=<ltrb 0 0 0 0>
      outlineSettings {
       bUseBrushTransparency=false
       roundingType=HalfHeightRadius
       width=0
       color {colorUseRule=UseColor_Specified, specifiedColor=<rgba 0 0 0 0>}
       cornerRadii=<0 0 0 0>
      }
      tintColor {colorUseRule=UseColor_Specified, specifiedColor=<rgba 1 1 1 1>}
      uVRegion {bIsValid=false, max=<0 0>, min=<0 0>}
     }
     parentRowBackgroundHoveredBrush {
      bIsDynamicallyLoaded=false
      drawAs=Image
      imageType=NoImage
      mirroring=NoMirror
      resourceName=None
      resourceObject=null
      tiling=NoTile
      imageSize=<32 32>
      margin=<ltrb 0 0 0 0>
      outlineSettings {
       bUseBrushTransparency=false
       roundingType=HalfHeightRadius
       width=0
       color {colorUseRule=UseColor_Specified, specifiedColor=<rgba 0 0 0 0>}
       cornerRadii=<0 0 0 0>
      }
      tintColor {colorUseRule=UseColor_Specified, specifiedColor=<rgba 1 1 1 1>}
      uVRegion {bIsValid=false, max=<0 0>, min=<0 0>}
     }
     selectedTextColor {colorUseRule=UseColor_Specified, specifiedColor=<rgba 1 1 1 1>}
     selectorFocusedBrush {
      bIsDynamicallyLoaded=false
      drawAs=Border
      imageType=FullColor
      mirroring=NoMirror
      resourceName=../../../Engine/Content/Slate/Common/Selector.png
      resourceObject=null
      tiling=Both
      imageSize=<0 0>
      margin=<ltrb 0.25 0.25 0.25 0.25>
      outlineSettings {
       bUseBrushTransparency=false
       roundingType=HalfHeightRadius
       width=0
       color {colorUseRule=UseColor_Specified, specifiedColor=<rgba 0 0 0 0>}
       cornerRadii=<0 0 0 0>
      }
      tintColor {colorUseRule=UseColor_Specified, specifiedColor=<rgba 0.701 0.225 0.003 1>}
      uVRegion {bIsValid=false, max=<0 0>, min=<0 0>}
     }
     textColor {colorUseRule=UseColor_Specified, specifiedColor=<rgba 0.5271 0.5271 0.5271 1>}
    }
    widgetStyle {
     comboButtonStyle {
      downArrowAlign=VAlign_Center
      buttonStyle {
       clickedSlateSound {resourceObject=null}
       disabled {
        bIsDynamicallyLoaded=false
        drawAs=RoundedBox
        imageType=NoImage
        mirroring=NoMirror
        resourceName=None
        resourceObject=null
        tiling=NoTile
        imageSize=<0 0>
        margin=<ltrb 0 0 0 0>
        outlineSettings {
         bUseBrushTransparency=false
         roundingType=FixedRadius
         width=1
         color {colorUseRule=UseColor_ColorTable, specifiedColor=<rgba 0 0 0 0>}
         cornerRadii=<4 4 4 4>
        }
        tintColor {colorUseRule=UseColor_Specified, specifiedColor=<rgba 0.0395 0.0395 0.0395 1>}
        uVRegion {bIsValid=false, max=<0 0>, min=<0 0>}
       }
       disabledForeground {colorUseRule=UseColor_ColorTable, specifiedColor=<rgba 0 0 0 0>}
       hovered {
        bIsDynamicallyLoaded=false
        drawAs=RoundedBox
        imageType=NoImage
        mirroring=NoMirror
        resourceName=None
        resourceObject=null
        tiling=NoTile
        imageSize=<0 0>
        margin=<ltrb 0 0 0 0>
        outlineSettings {
         bUseBrushTransparency=false
         roundingType=FixedRadius
         width=1
         color {colorUseRule=UseColor_ColorTable, specifiedColor=<rgba 0 0 0 0>}
         cornerRadii=<4 4 4 4>
        }
        tintColor {colorUseRule=UseColor_Specified, specifiedColor=<rgba 0.0048 0.0048 0.0048 1>}
        uVRegion {bIsValid=false, max=<0 0>, min=<0 0>}
       }
       hoveredForeground {colorUseRule=UseColor_ColorTable, specifiedColor=<rgba 0 0 0 0>}
       hoveredSlateSound {resourceObject=null}
       normal {
        bIsDynamicallyLoaded=false
        drawAs=RoundedBox
        imageType=NoImage
        mirroring=NoMirror
        resourceName=None
        resourceObject=null
        tiling=NoTile
        imageSize=<0 0>
        margin=<ltrb 0 0 0 0>
        outlineSettings {
         bUseBrushTransparency=false
         roundingType=FixedRadius
         width=1
         color {colorUseRule=UseColor_ColorTable, specifiedColor=<rgba 0 0 0 0>}
         cornerRadii=<4 4 4 4>
        }
        tintColor {colorUseRule=UseColor_Specified, specifiedColor=<rgba 0.0048 0.0048 0.0048 1>}
        uVRegion {bIsValid=false, max=<0 0>, min=<0 0>}
       }
       normalForeground {colorUseRule=UseColor_ColorTable, specifiedColor=<rgba 0 0 0 0>}
       normalPadding=<ltrb 8 1 8 1>
       pressed {
        bIsDynamicallyLoaded=false
        drawAs=RoundedBox
        imageType=NoImage
        mirroring=NoMirror
        resourceName=None
        resourceObject=null
        tiling=NoTile
        imageSize=<0 0>
        margin=<ltrb 0 0 0 0>
        outlineSettings {
         bUseBrushTransparency=false
         roundingType=FixedRadius
         width=1
         color {colorUseRule=UseColor_ColorTable, specifiedColor=<rgba 0 0 0 0>}
         cornerRadii=<4 4 4 4>
        }
        tintColor {colorUseRule=UseColor_Specified, specifiedColor=<rgba 0.0048 0.0048 0.0048 1>}
        uVRegion {bIsValid=false, max=<0 0>, min=<0 0>}
       }
       pressedForeground {colorUseRule=UseColor_ColorTable, specifiedColor=<rgba 0 0 0 0>}
       pressedPadding=<ltrb 8 2 8 0>
       pressedSlateSound {resourceObject=null}
      }
      contentPadding=<ltrb 0 0 0 0>
      downArrowImage {
       bIsDynamicallyLoaded=false
       drawAs=Image
       imageType=Vector
       mirroring=NoMirror
       resourceName=../../../Engine/Content/Slate/Starship/CoreWidgets/ComboBox/wide-chevron-down.svg
       resourceObject=null
       tiling=NoTile
       imageSize=<16 16>
       margin=<ltrb 0 0 0 0>
       outlineSettings {
        bUseBrushTransparency=false
        roundingType=HalfHeightRadius
        width=0
        color {colorUseRule=UseColor_Specified, specifiedColor=<rgba 0 0 0 0>}
        cornerRadii=<0 0 0 0>
       }
       tintColor {colorUseRule=UseColor_Specified, specifiedColor=<rgba 1 1 1 1>}
       uVRegion {bIsValid=false, max=<0 0>, min=<0 0>}
      }
      downArrowPadding=<ltrb 2 2 2 2>
      menuBorderBrush {
       bIsDynamicallyLoaded=false
       drawAs=RoundedBox
       imageType=NoImage
       mirroring=NoMirror
       resourceName=None
       resourceObject=null
       tiling=NoTile
       imageSize=<0 0>
       margin=<ltrb 0 0 0 0>
       outlineSettings {
        bUseBrushTransparency=false
        roundingType=FixedRadius
        width=1
        color {colorUseRule=UseColor_ColorTable, specifiedColor=<rgba 0 0 0 0>}
        cornerRadii=<0 0 0 0>
       }
       tintColor {colorUseRule=UseColor_Specified, specifiedColor=<rgba 0.0395 0.0395 0.0395 1>}
       uVRegion {bIsValid=false, max=<0 0>, min=<0 0>}
      }
      menuBorderPadding=<ltrb 0 0 0 0>
      shadowColorAndOpacity=<rgba 0 0 0 1>
      shadowOffset=<0 0>
     }
     contentPadding=<ltrb 0 0 0 0>
     menuRowPadding=<ltrb 8 3 8 3>
     pressedSlateSound {resourceObject=null}
     selectionChangeSlateSound {resourceObject=null}
    }
   }
   slot {class=HorizontalBoxSlot, properties {padding=<ltrb 5 0 0 0>, size {sizeRule=Fill, value=1}}}
  }
  PluginHorizontalBox {
   widgetClassPath=&HorizontalBox
   parent=VerticalBox_4
   widget=PluginHorizontalBox
   properties {toolTipText=Where should the files be saved?}
   slot {class=VerticalBoxSlot, properties {padding=<ltrb 0 1 0 1>}}
  }
  SizeBox_5 {
   widgetClassPath=&SizeBox
   parent=PluginHorizontalBox
   widget=SizeBox_5
   properties {bOverride_MinDesiredWidth=true, minDesiredWidth=100}
   slot {class=HorizontalBoxSlot, properties {verticalAlignment=VAlign_Center}}
  }
  PluginTextBlock {
   widgetClassPath=&TextBlock
   bIsVariable=true
   parent=SizeBox_5
   widget=PluginTextBlock
   properties {
    text=Plugin
    colorAndOpacity {colorUseRule=UseColor_Specified, specifiedColor=<rgba 0.5255 0.5255 0.5255 1>}
    font {
     bForceMonospaced=false
     bMaterialIsStencil=false
     fontMaterial=null
     fontObject=&Roboto
     letterSpacing=0
     monospacedWidth=1
     size=10
     skewAmount=0
     typefaceFontName=Bold
     outlineSettings {
      bApplyOutlineToDropShadows=false
      bMiteredCorners=false
      bSeparateFillAlpha=false
      outlineMaterial=null
      outlineSize=0
      outlineColor=<rgba 0 0 0 1>
     }
    }
   }
   slot {class=SizeBoxSlot, properties {horizontalAlignment=HAlign_Right}}
  }
  PluginComboBox {
   widgetClassPath=&SearchableStringComboBox
   bIsVariable=true
   parent=PluginHorizontalBox
   widget=PluginComboBox
   properties {
    font {
     bForceMonospaced=false
     bMaterialIsStencil=false
     fontMaterial=null
     fontObject=&Roboto
     letterSpacing=0
     monospacedWidth=1
     size=10
     skewAmount=0
     typefaceFontName=Regular
     outlineSettings {
      bApplyOutlineToDropShadows=false
      bMiteredCorners=false
      bSeparateFillAlpha=false
      outlineMaterial=null
      outlineSize=0
      outlineColor=<rgba 0 0 0 1>
     }
    }
    foregroundColor {colorUseRule=UseColor_Specified, specifiedColor=<rgba 0.5271 0.5271 0.5271 1>}
    itemStyle {
     bUseParentRowBrush=false
     activeBrush {
      bIsDynamicallyLoaded=false
      drawAs=Image
      imageType=NoImage
      mirroring=NoMirror
      resourceName=None
      resourceObject=null
      tiling=NoTile
      imageSize=<0 0>
      margin=<ltrb 0 0 0 0>
      outlineSettings {
       bUseBrushTransparency=false
       roundingType=HalfHeightRadius
       width=0
       color {colorUseRule=UseColor_Specified, specifiedColor=<rgba 0 0 0 0>}
       cornerRadii=<0 0 0 0>
      }
      tintColor {colorUseRule=UseColor_Specified, specifiedColor=<rgba 0 0.162 0.7454 1>}
      uVRegion {bIsValid=false, max=<0 0>, min=<0 0>}
     }
     activeHighlightedBrush {
      bIsDynamicallyLoaded=false
      drawAs=Image
      imageType=NoImage
      mirroring=NoMirror
      resourceName=None
      resourceObject=null
      tiling=NoTile
      imageSize=<0 0>
      margin=<ltrb 0 0 0 0>
      outlineSettings {
       bUseBrushTransparency=false
       roundingType=HalfHeightRadius
       width=0
       color {colorUseRule=UseColor_Specified, specifiedColor=<rgba 0 0 0 0>}
       cornerRadii=<0 0 0 0>
      }
      tintColor {colorUseRule=UseColor_Specified, specifiedColor=<rgba 0.0252 0.0319 0.0423 1>}
      uVRegion {bIsValid=false, max=<0 0>, min=<0 0>}
     }
     activeHoveredBrush {
      bIsDynamicallyLoaded=false
      drawAs=Image
      imageType=NoImage
      mirroring=NoMirror
      resourceName=None
      resourceObject=null
      tiling=NoTile
      imageSize=<0 0>
      margin=<ltrb 0 0 0 0>
      outlineSettings {
       bUseBrushTransparency=false
       roundingType=HalfHeightRadius
       width=0
       color {colorUseRule=UseColor_Specified, specifiedColor=<rgba 0 0 0 0>}
       cornerRadii=<0 0 0 0>
      }
      tintColor {colorUseRule=UseColor_Specified, specifiedColor=<rgba 0 0.162 0.7454 1>}
      uVRegion {bIsValid=false, max=<0 0>, min=<0 0>}
     }
     dropIndicator_Above {
      bIsDynamicallyLoaded=false
      drawAs=Box
      imageType=FullColor
      mirroring=NoMirror
      resourceName=../../../Engine/Content/Slate/Common/DropZoneIndicator_Above.png
      resourceObject=null
      tiling=NoTile
      imageSize=<0 0>
      margin=<ltrb 0.625 0.625 0 0>
      outlineSettings {
       bUseBrushTransparency=false
       roundingType=HalfHeightRadius
       width=0
       color {colorUseRule=UseColor_Specified, specifiedColor=<rgba 0 0 0 0>}
       cornerRadii=<0 0 0 0>
      }
      tintColor {colorUseRule=UseColor_Specified, specifiedColor=<rgba 0.0091 0.3515 0.9647 1>}
      uVRegion {bIsValid=false, max=<0 0>, min=<0 0>}
     }
     dropIndicator_Below {
      bIsDynamicallyLoaded=false
      drawAs=Box
      imageType=FullColor
      mirroring=NoMirror
      resourceName=../../../Engine/Content/Slate/Common/DropZoneIndicator_Below.png
      resourceObject=null
      tiling=NoTile
      imageSize=<0 0>
      margin=<ltrb 0.625 0 0 0.625>
      outlineSettings {
       bUseBrushTransparency=false
       roundingType=HalfHeightRadius
       width=0
       color {colorUseRule=UseColor_Specified, specifiedColor=<rgba 0 0 0 0>}
       cornerRadii=<0 0 0 0>
      }
      tintColor {colorUseRule=UseColor_Specified, specifiedColor=<rgba 0.0091 0.3515 0.9647 1>}
      uVRegion {bIsValid=false, max=<0 0>, min=<0 0>}
     }
     dropIndicator_Onto {
      bIsDynamicallyLoaded=false
      drawAs=Box
      imageType=FullColor
      mirroring=NoMirror
      resourceName=../../../Engine/Content/Slate/Common/DropZoneIndicator_Onto.png
      resourceObject=null
      tiling=NoTile
      imageSize=<0 0>
      margin=<ltrb 0.25 0.25 0.25 0.25>
      outlineSettings {
       bUseBrushTransparency=false
       roundingType=HalfHeightRadius
       width=0
       color {colorUseRule=UseColor_Specified, specifiedColor=<rgba 0 0 0 0>}
       cornerRadii=<0 0 0 0>
      }
      tintColor {colorUseRule=UseColor_Specified, specifiedColor=<rgba 0.0091 0.3515 0.9647 1>}
      uVRegion {bIsValid=false, max=<0 0>, min=<0 0>}
     }
     evenRowBackgroundBrush {
      bIsDynamicallyLoaded=false
      drawAs=Image
      imageType=NoImage
      mirroring=NoMirror
      resourceName=None
      resourceObject=null
      tiling=NoTile
      imageSize=<0 0>
      margin=<ltrb 0 0 0 0>
      outlineSettings {
       bUseBrushTransparency=false
       roundingType=HalfHeightRadius
       width=0
       color {colorUseRule=UseColor_Specified, specifiedColor=<rgba 0 0 0 0>}
       cornerRadii=<0 0 0 0>
      }
      tintColor {colorUseRule=UseColor_Specified, specifiedColor=<rgba 0.0103 0.0103 0.0103 1>}
      uVRegion {bIsValid=false, max=<0 0>, min=<0 0>}
     }
     evenRowBackgroundHoveredBrush {
      bIsDynamicallyLoaded=false
      drawAs=Image
      imageType=NoImage
      mirroring=NoMirror
      resourceName=None
      resourceObject=null
      tiling=NoTile
      imageSize=<0 0>
      margin=<ltrb 0 0 0 0>
      outlineSettings {
       bUseBrushTransparency=false
       roundingType=HalfHeightRadius
       width=0
       color {colorUseRule=UseColor_Specified, specifiedColor=<rgba 0 0 0 0>}
       cornerRadii=<0 0 0 0>
      }
      tintColor {colorUseRule=UseColor_Specified, specifiedColor=<rgba 0.0176 0.0176 0.0176 1>}
      uVRegion {bIsValid=false, max=<0 0>, min=<0 0>}
     }
     inactiveBrush {
      bIsDynamicallyLoaded=false
      drawAs=Image
      imageType=NoImage
      mirroring=NoMirror
      resourceName=None
      resourceObject=null
      tiling=NoTile
      imageSize=<0 0>
      margin=<ltrb 0 0 0 0>
      outlineSettings {
       bUseBrushTransparency=false
       roundingType=HalfHeightRadius
       width=0
       color {colorUseRule=UseColor_Specified, specifiedColor=<rgba 0 0 0 0>}
       cornerRadii=<0 0 0 0>
      }
      tintColor {colorUseRule=UseColor_Specified, specifiedColor=<rgba 0.0513 0.0953 0.159 1>}
      uVRegion {bIsValid=false, max=<0 0>, min=<0 0>}
     }
     inactiveHighlightedBrush {
      bIsDynamicallyLoaded=false
      drawAs=Image
      imageType=NoImage
      mirroring=NoMirror
      resourceName=None
      resourceObject=null
      tiling=NoTile
      imageSize=<0 0>
      margin=<ltrb 0 0 0 0>
      outlineSettings {
       bUseBrushTransparency=false
       roundingType=HalfHeightRadius
       width=0
       color {colorUseRule=UseColor_Specified, specifiedColor=<rgba 0 0 0 0>}
       cornerRadii=<0 0 0 0>
      }
      tintColor {colorUseRule=UseColor_Specified, specifiedColor=<rgba 0.0252 0.0319 0.0423 1>}
      uVRegion {bIsValid=false, max=<0 0>, min=<0 0>}
     }
     inactiveHoveredBrush {
      bIsDynamicallyLoaded=false
      drawAs=Image
      imageType=NoImage
      mirroring=NoMirror
      resourceName=None
      resourceObject=null
      tiling=NoTile
      imageSize=<0 0>
      margin=<ltrb 0 0 0 0>
      outlineSettings {
       bUseBrushTransparency=false
       roundingType=HalfHeightRadius
       width=0
       color {colorUseRule=UseColor_Specified, specifiedColor=<rgba 0 0 0 0>}
       cornerRadii=<0 0 0 0>
      }
      tintColor {colorUseRule=UseColor_Specified, specifiedColor=<rgba 0.0513 0.0953 0.159 1>}
      uVRegion {bIsValid=false, max=<0 0>, min=<0 0>}
     }
     oddRowBackgroundBrush {
      bIsDynamicallyLoaded=false
      drawAs=Image
      imageType=NoImage
      mirroring=NoMirror
      resourceName=None
      resourceObject=null
      tiling=NoTile
      imageSize=<0 0>
      margin=<ltrb 0 0 0 0>
      outlineSettings {
       bUseBrushTransparency=false
       roundingType=HalfHeightRadius
       width=0
       color {colorUseRule=UseColor_Specified, specifiedColor=<rgba 0 0 0 0>}
       cornerRadii=<0 0 0 0>
      }
      tintColor {colorUseRule=UseColor_Specified, specifiedColor=<rgba 0.0103 0.0103 0.0103 1>}
      uVRegion {bIsValid=false, max=<0 0>, min=<0 0>}
     }
     oddRowBackgroundHoveredBrush {
      bIsDynamicallyLoaded=false
      drawAs=Image
      imageType=NoImage
      mirroring=NoMirror
      resourceName=None
      resourceObject=null
      tiling=NoTile
      imageSize=<0 0>
      margin=<ltrb 0 0 0 0>
      outlineSettings {
       bUseBrushTransparency=false
       roundingType=HalfHeightRadius
       width=0
       color {colorUseRule=UseColor_Specified, specifiedColor=<rgba 0 0 0 0>}
       cornerRadii=<0 0 0 0>
      }
      tintColor {colorUseRule=UseColor_Specified, specifiedColor=<rgba 0.0176 0.0176 0.0176 1>}
      uVRegion {bIsValid=false, max=<0 0>, min=<0 0>}
     }
     parentRowBackgroundBrush {
      bIsDynamicallyLoaded=false
      drawAs=Image
      imageType=NoImage
      mirroring=NoMirror
      resourceName=None
      resourceObject=null
      tiling=NoTile
      imageSize=<32 32>
      margin=<ltrb 0 0 0 0>
      outlineSettings {
       bUseBrushTransparency=false
       roundingType=HalfHeightRadius
       width=0
       color {colorUseRule=UseColor_Specified, specifiedColor=<rgba 0 0 0 0>}
       cornerRadii=<0 0 0 0>
      }
      tintColor {colorUseRule=UseColor_Specified, specifiedColor=<rgba 1 1 1 1>}
      uVRegion {bIsValid=false, max=<0 0>, min=<0 0>}
     }
     parentRowBackgroundHoveredBrush {
      bIsDynamicallyLoaded=false
      drawAs=Image
      imageType=NoImage
      mirroring=NoMirror
      resourceName=None
      resourceObject=null
      tiling=NoTile
      imageSize=<32 32>
      margin=<ltrb 0 0 0 0>
      outlineSettings {
       bUseBrushTransparency=false
       roundingType=HalfHeightRadius
       width=0
       color {colorUseRule=UseColor_Specified, specifiedColor=<rgba 0 0 0 0>}
       cornerRadii=<0 0 0 0>
      }
      tintColor {colorUseRule=UseColor_Specified, specifiedColor=<rgba 1 1 1 1>}
      uVRegion {bIsValid=false, max=<0 0>, min=<0 0>}
     }
     selectedTextColor {colorUseRule=UseColor_Specified, specifiedColor=<rgba 1 1 1 1>}
     selectorFocusedBrush {
      bIsDynamicallyLoaded=false
      drawAs=Border
      imageType=FullColor
      mirroring=NoMirror
      resourceName=../../../Engine/Content/Slate/Common/Selector.png
      resourceObject=null
      tiling=Both
      imageSize=<0 0>
      margin=<ltrb 0.25 0.25 0.25 0.25>
      outlineSettings {
       bUseBrushTransparency=false
       roundingType=HalfHeightRadius
       width=0
       color {colorUseRule=UseColor_Specified, specifiedColor=<rgba 0 0 0 0>}
       cornerRadii=<0 0 0 0>
      }
      tintColor {colorUseRule=UseColor_Specified, specifiedColor=<rgba 0.701 0.225 0.003 1>}
      uVRegion {bIsValid=false, max=<0 0>, min=<0 0>}
     }
     textColor {colorUseRule=UseColor_Specified, specifiedColor=<rgba 0.5271 0.5271 0.5271 1>}
    }
    widgetStyle {
     comboButtonStyle {
      downArrowAlign=VAlign_Center
      buttonStyle {
       clickedSlateSound {resourceObject=null}
       disabled {
        bIsDynamicallyLoaded=false
        drawAs=RoundedBox
        imageType=NoImage
        mirroring=NoMirror
        resourceName=None
        resourceObject=null
        tiling=NoTile
        imageSize=<0 0>
        margin=<ltrb 0 0 0 0>
        outlineSettings {
         bUseBrushTransparency=false
         roundingType=FixedRadius
         width=1
         color {colorUseRule=UseColor_ColorTable, specifiedColor=<rgba 0 0 0 0>}
         cornerRadii=<4 4 4 4>
        }
        tintColor {colorUseRule=UseColor_Specified, specifiedColor=<rgba 0.0395 0.0395 0.0395 1>}
        uVRegion {bIsValid=false, max=<0 0>, min=<0 0>}
       }
       disabledForeground {colorUseRule=UseColor_ColorTable, specifiedColor=<rgba 0 0 0 0>}
       hovered {
        bIsDynamicallyLoaded=false
        drawAs=RoundedBox
        imageType=NoImage
        mirroring=NoMirror
        resourceName=None
        resourceObject=null
        tiling=NoTile
        imageSize=<0 0>
        margin=<ltrb 0 0 0 0>
        outlineSettings {
         bUseBrushTransparency=false
         roundingType=FixedRadius
         width=1
         color {colorUseRule=UseColor_ColorTable, specifiedColor=<rgba 0 0 0 0>}
         cornerRadii=<4 4 4 4>
        }
        tintColor {colorUseRule=UseColor_Specified, specifiedColor=<rgba 0.0048 0.0048 0.0048 1>}
        uVRegion {bIsValid=false, max=<0 0>, min=<0 0>}
       }
       hoveredForeground {colorUseRule=UseColor_ColorTable, specifiedColor=<rgba 0 0 0 0>}
       hoveredSlateSound {resourceObject=null}
       normal {
        bIsDynamicallyLoaded=false
        drawAs=RoundedBox
        imageType=NoImage
        mirroring=NoMirror
        resourceName=None
        resourceObject=null
        tiling=NoTile
        imageSize=<0 0>
        margin=<ltrb 0 0 0 0>
        outlineSettings {
         bUseBrushTransparency=false
         roundingType=FixedRadius
         width=1
         color {colorUseRule=UseColor_ColorTable, specifiedColor=<rgba 0 0 0 0>}
         cornerRadii=<4 4 4 4>
        }
        tintColor {colorUseRule=UseColor_Specified, specifiedColor=<rgba 0.0048 0.0048 0.0048 1>}
        uVRegion {bIsValid=false, max=<0 0>, min=<0 0>}
       }
       normalForeground {colorUseRule=UseColor_ColorTable, specifiedColor=<rgba 0 0 0 0>}
       normalPadding=<ltrb 8 1 8 1>
       pressed {
        bIsDynamicallyLoaded=false
        drawAs=RoundedBox
        imageType=NoImage
        mirroring=NoMirror
        resourceName=None
        resourceObject=null
        tiling=NoTile
        imageSize=<0 0>
        margin=<ltrb 0 0 0 0>
        outlineSettings {
         bUseBrushTransparency=false
         roundingType=FixedRadius
         width=1
         color {colorUseRule=UseColor_ColorTable, specifiedColor=<rgba 0 0 0 0>}
         cornerRadii=<4 4 4 4>
        }
        tintColor {colorUseRule=UseColor_Specified, specifiedColor=<rgba 0.0048 0.0048 0.0048 1>}
        uVRegion {bIsValid=false, max=<0 0>, min=<0 0>}
       }
       pressedForeground {colorUseRule=UseColor_ColorTable, specifiedColor=<rgba 0 0 0 0>}
       pressedPadding=<ltrb 8 2 8 0>
       pressedSlateSound {resourceObject=null}
      }
      contentPadding=<ltrb 0 0 0 0>
      downArrowImage {
       bIsDynamicallyLoaded=false
       drawAs=Image
       imageType=Vector
       mirroring=NoMirror
       resourceName=../../../Engine/Content/Slate/Starship/CoreWidgets/ComboBox/wide-chevron-down.svg
       resourceObject=null
       tiling=NoTile
       imageSize=<16 16>
       margin=<ltrb 0 0 0 0>
       outlineSettings {
        bUseBrushTransparency=false
        roundingType=HalfHeightRadius
        width=0
        color {colorUseRule=UseColor_Specified, specifiedColor=<rgba 0 0 0 0>}
        cornerRadii=<0 0 0 0>
       }
       tintColor {colorUseRule=UseColor_Specified, specifiedColor=<rgba 1 1 1 1>}
       uVRegion {bIsValid=false, max=<0 0>, min=<0 0>}
      }
      downArrowPadding=<ltrb 2 2 2 2>
      menuBorderBrush {
       bIsDynamicallyLoaded=false
       drawAs=RoundedBox
       imageType=NoImage
       mirroring=NoMirror
       resourceName=None
       resourceObject=null
       tiling=NoTile
       imageSize=<0 0>
       margin=<ltrb 0 0 0 0>
       outlineSettings {
        bUseBrushTransparency=false
        roundingType=FixedRadius
        width=1
        color {colorUseRule=UseColor_ColorTable, specifiedColor=<rgba 0 0 0 0>}
        cornerRadii=<0 0 0 0>
       }
       tintColor {colorUseRule=UseColor_Specified, specifiedColor=<rgba 0.0395 0.0395 0.0395 1>}
       uVRegion {bIsValid=false, max=<0 0>, min=<0 0>}
      }
      menuBorderPadding=<ltrb 0 0 0 0>
      shadowColorAndOpacity=<rgba 0 0 0 1>
      shadowOffset=<0 0>
     }
     contentPadding=<ltrb 0 0 0 0>
     menuRowPadding=<ltrb 8 3 8 3>
     pressedSlateSound {resourceObject=null}
     selectionChangeSlateSound {resourceObject=null}
    }
   }
   slot {class=HorizontalBoxSlot, properties {padding=<ltrb 5 0 0 0>, size {sizeRule=Fill, value=1}}}
  }
  SubfolderHorizontalBox {
   widgetClassPath=&HorizontalBox
   parent=VerticalBox_4
   widget=SubfolderHorizontalBox
   slot {class=VerticalBoxSlot, properties {padding=<ltrb 0 1 0 1>}}
  }
  SizeBox_2 {
   widgetClassPath=&SizeBox
   parent=SubfolderHorizontalBox
   widget=SizeBox_2
   properties {bOverride_MinDesiredWidth=true, minDesiredWidth=100}
   slot {class=HorizontalBoxSlot, properties {verticalAlignment=VAlign_Center}}
  }
  SubfolderTextBlock {
   widgetClassPath=&TextBlock
   bIsVariable=true
   parent=SizeBox_2
   widget=SubfolderTextBlock
   properties {
    text=Subfolder
    colorAndOpacity {colorUseRule=UseColor_Specified, specifiedColor=<rgba 0.5255 0.5255 0.5255 1>}
    font {
     bForceMonospaced=false
     bMaterialIsStencil=false
     fontMaterial=null
     fontObject=&Roboto
     letterSpacing=0
     monospacedWidth=1
     size=10
     skewAmount=0
     typefaceFontName=Bold
     outlineSettings {
      bApplyOutlineToDropShadows=false
      bMiteredCorners=false
      bSeparateFillAlpha=false
      outlineMaterial=null
      outlineSize=0
      outlineColor=<rgba 0 0 0 1>
     }
    }
   }
   slot {class=SizeBoxSlot, properties {horizontalAlignment=HAlign_Right}}
  }
  SubfolderCustomEntryComboBox {
   widgetClassPath=&/EUW_CustomEntryComboBox
   bIsVariable=true
   parent=SubfolderHorizontalBox
   widget=SubfolderCustomEntryComboBox
   properties {toolTipText="Technically speaking, what is this File for/How Is it used?"}
   slot {class=HorizontalBoxSlot, properties {size {sizeRule=Fill, value=1}}}
  }
  AssetNameBorder {
   widgetClassPath=&Border
   bIsVariable=true
   parent=EditorUtilityScrollBox_0
   widget=AssetNameBorder
   properties {brushColor=<rgba 0.0273 0.0273 0.0273 1>}
   slot {class=ScrollBoxSlot, properties {padding=<ltrb 0 2.5 0 2.5>}}
  }
  NameVerticalBox {widgetClassPath=&VerticalBox, parent=AssetNameBorder, widget=NameVerticalBox, slot {class=BorderSlot}}
  PluginTextBlock_1 {
   widgetClassPath=&TextBlock
   parent=NameVerticalBox
   widget=PluginTextBlock_1
   properties {
    text=File Name
    colorAndOpacity {colorUseRule=UseColor_Specified, specifiedColor=<rgba 0.5255 0.5255 0.5255 1>}
    font {
     bForceMonospaced=false
     bMaterialIsStencil=false
     fontMaterial=null
     fontObject=&Roboto
     letterSpacing=0
     monospacedWidth=1
     size=9
     skewAmount=0
     typefaceFontName=Bold
     outlineSettings {
      bApplyOutlineToDropShadows=false
      bMiteredCorners=false
      bSeparateFillAlpha=false
      outlineMaterial=null
      outlineSize=0
      outlineColor=<rgba 0 0 0 1>
     }
    }
   }
   slot {class=VerticalBoxSlot, properties {horizontalAlignment=HAlign_Left, padding=<ltrb 0 2 0 0>}}
  }
  AnimSetHorizontalBox {
   widgetClassPath=&HorizontalBox
   bIsVariable=true
   parent=NameVerticalBox
   widget=AnimSetHorizontalBox
   slot {class=VerticalBoxSlot, properties {padding=<ltrb 0 2 0 2>}}
  }
  SizeBox_3 {
   widgetClassPath=&SizeBox
   parent=AnimSetHorizontalBox
   widget=SizeBox_3
   properties {bOverride_MinDesiredWidth=true, minDesiredWidth=100}
   slot {class=HorizontalBoxSlot, properties {verticalAlignment=VAlign_Center}}
  }
  CategoryTextBlock {
   widgetClassPath=&TextBlock
   bIsVariable=true
   parent=SizeBox_3
   widget=CategoryTextBlock
   properties {
    text=Animation Set
    toolTipText=Is this for a specific Item?
    colorAndOpacity {colorUseRule=UseColor_Specified, specifiedColor=<rgba 0 0.5255 0.3373 1>}
    font {
     bForceMonospaced=false
     bMaterialIsStencil=false
     fontMaterial=null
     fontObject=&Roboto
     letterSpacing=0
     monospacedWidth=1
     size=10
     skewAmount=0
     typefaceFontName=Bold
     outlineSettings {
      bApplyOutlineToDropShadows=false
      bMiteredCorners=false
      bSeparateFillAlpha=false
      outlineMaterial=null
      outlineSize=0
      outlineColor=<rgba 0 0 0 1>
     }
    }
   }
   slot {class=SizeBoxSlot, properties {horizontalAlignment=HAlign_Right}}
  }
  VerticalBox_3 {
   widgetClassPath=&VerticalBox
   parent=AnimSetHorizontalBox
   widget=VerticalBox_3
   slot {class=HorizontalBoxSlot, properties {padding=<ltrb 5 0 0 0>, size {sizeRule=Fill, value=1}}}
  }
  AnimSetComboBox {
   widgetClassPath=&SearchableStringComboBox
   bIsVariable=true
   parent=VerticalBox_3
   widget=AnimSetComboBox
   properties {
    font {
     bForceMonospaced=false
     bMaterialIsStencil=false
     fontMaterial=null
     fontObject=&Roboto
     letterSpacing=0
     monospacedWidth=1
     size=10
     skewAmount=0
     typefaceFontName=Regular
     outlineSettings {
      bApplyOutlineToDropShadows=false
      bMiteredCorners=false
      bSeparateFillAlpha=false
      outlineMaterial=null
      outlineSize=0
      outlineColor=<rgba 0 0 0 1>
     }
    }
    foregroundColor {colorUseRule=UseColor_Specified, specifiedColor=<rgba 0.5271 0.5271 0.5271 1>}
    itemStyle {
     bUseParentRowBrush=false
     activeBrush {
      bIsDynamicallyLoaded=false
      drawAs=Image
      imageType=NoImage
      mirroring=NoMirror
      resourceName=None
      resourceObject=null
      tiling=NoTile
      imageSize=<0 0>
      margin=<ltrb 0 0 0 0>
      outlineSettings {
       bUseBrushTransparency=false
       roundingType=HalfHeightRadius
       width=0
       color {colorUseRule=UseColor_Specified, specifiedColor=<rgba 0 0 0 0>}
       cornerRadii=<0 0 0 0>
      }
      tintColor {colorUseRule=UseColor_Specified, specifiedColor=<rgba 0 0.162 0.7454 1>}
      uVRegion {bIsValid=false, max=<0 0>, min=<0 0>}
     }
     activeHighlightedBrush {
      bIsDynamicallyLoaded=false
      drawAs=Image
      imageType=NoImage
      mirroring=NoMirror
      resourceName=None
      resourceObject=null
      tiling=NoTile
      imageSize=<0 0>
      margin=<ltrb 0 0 0 0>
      outlineSettings {
       bUseBrushTransparency=false
       roundingType=HalfHeightRadius
       width=0
       color {colorUseRule=UseColor_Specified, specifiedColor=<rgba 0 0 0 0>}
       cornerRadii=<0 0 0 0>
      }
      tintColor {colorUseRule=UseColor_Specified, specifiedColor=<rgba 0.0252 0.0319 0.0423 1>}
      uVRegion {bIsValid=false, max=<0 0>, min=<0 0>}
     }
     activeHoveredBrush {
      bIsDynamicallyLoaded=false
      drawAs=Image
      imageType=NoImage
      mirroring=NoMirror
      resourceName=None
      resourceObject=null
      tiling=NoTile
      imageSize=<0 0>
      margin=<ltrb 0 0 0 0>
      outlineSettings {
       bUseBrushTransparency=false
       roundingType=HalfHeightRadius
       width=0
       color {colorUseRule=UseColor_Specified, specifiedColor=<rgba 0 0 0 0>}
       cornerRadii=<0 0 0 0>
      }
      tintColor {colorUseRule=UseColor_Specified, specifiedColor=<rgba 0 0.162 0.7454 1>}
      uVRegion {bIsValid=false, max=<0 0>, min=<0 0>}
     }
     dropIndicator_Above {
      bIsDynamicallyLoaded=false
      drawAs=Box
      imageType=FullColor
      mirroring=NoMirror
      resourceName=../../../Engine/Content/Slate/Common/DropZoneIndicator_Above.png
      resourceObject=null
      tiling=NoTile
      imageSize=<0 0>
      margin=<ltrb 0.625 0.625 0 0>
      outlineSettings {
       bUseBrushTransparency=false
       roundingType=HalfHeightRadius
       width=0
       color {colorUseRule=UseColor_Specified, specifiedColor=<rgba 0 0 0 0>}
       cornerRadii=<0 0 0 0>
      }
      tintColor {colorUseRule=UseColor_Specified, specifiedColor=<rgba 0.0091 0.3515 0.9647 1>}
      uVRegion {bIsValid=false, max=<0 0>, min=<0 0>}
     }
     dropIndicator_Below {
      bIsDynamicallyLoaded=false
      drawAs=Box
      imageType=FullColor
      mirroring=NoMirror
      resourceName=../../../Engine/Content/Slate/Common/DropZoneIndicator_Below.png
      resourceObject=null
      tiling=NoTile
      imageSize=<0 0>
      margin=<ltrb 0.625 0 0 0.625>
      outlineSettings {
       bUseBrushTransparency=false
       roundingType=HalfHeightRadius
       width=0
       color {colorUseRule=UseColor_Specified, specifiedColor=<rgba 0 0 0 0>}
       cornerRadii=<0 0 0 0>
      }
      tintColor {colorUseRule=UseColor_Specified, specifiedColor=<rgba 0.0091 0.3515 0.9647 1>}
      uVRegion {bIsValid=false, max=<0 0>, min=<0 0>}
     }
     dropIndicator_Onto {
      bIsDynamicallyLoaded=false
      drawAs=Box
      imageType=FullColor
      mirroring=NoMirror
      resourceName=../../../Engine/Content/Slate/Common/DropZoneIndicator_Onto.png
      resourceObject=null
      tiling=NoTile
      imageSize=<0 0>
      margin=<ltrb 0.25 0.25 0.25 0.25>
      outlineSettings {
       bUseBrushTransparency=false
       roundingType=HalfHeightRadius
       width=0
       color {colorUseRule=UseColor_Specified, specifiedColor=<rgba 0 0 0 0>}
       cornerRadii=<0 0 0 0>
      }
      tintColor {colorUseRule=UseColor_Specified, specifiedColor=<rgba 0.0091 0.3515 0.9647 1>}
      uVRegion {bIsValid=false, max=<0 0>, min=<0 0>}
     }
     evenRowBackgroundBrush {
      bIsDynamicallyLoaded=false
      drawAs=Image
      imageType=NoImage
      mirroring=NoMirror
      resourceName=None
      resourceObject=null
      tiling=NoTile
      imageSize=<0 0>
      margin=<ltrb 0 0 0 0>
      outlineSettings {
       bUseBrushTransparency=false
       roundingType=HalfHeightRadius
       width=0
       color {colorUseRule=UseColor_Specified, specifiedColor=<rgba 0 0 0 0>}
       cornerRadii=<0 0 0 0>
      }
      tintColor {colorUseRule=UseColor_Specified, specifiedColor=<rgba 0.0103 0.0103 0.0103 1>}
      uVRegion {bIsValid=false, max=<0 0>, min=<0 0>}
     }
     evenRowBackgroundHoveredBrush {
      bIsDynamicallyLoaded=false
      drawAs=Image
      imageType=NoImage
      mirroring=NoMirror
      resourceName=None
      resourceObject=null
      tiling=NoTile
      imageSize=<0 0>
      margin=<ltrb 0 0 0 0>
      outlineSettings {
       bUseBrushTransparency=false
       roundingType=HalfHeightRadius
       width=0
       color {colorUseRule=UseColor_Specified, specifiedColor=<rgba 0 0 0 0>}
       cornerRadii=<0 0 0 0>
      }
      tintColor {colorUseRule=UseColor_Specified, specifiedColor=<rgba 0.0176 0.0176 0.0176 1>}
      uVRegion {bIsValid=false, max=<0 0>, min=<0 0>}
     }
     inactiveBrush {
      bIsDynamicallyLoaded=false
      drawAs=Image
      imageType=NoImage
      mirroring=NoMirror
      resourceName=None
      resourceObject=null
      tiling=NoTile
      imageSize=<0 0>
      margin=<ltrb 0 0 0 0>
      outlineSettings {
       bUseBrushTransparency=false
       roundingType=HalfHeightRadius
       width=0
       color {colorUseRule=UseColor_Specified, specifiedColor=<rgba 0 0 0 0>}
       cornerRadii=<0 0 0 0>
      }
      tintColor {colorUseRule=UseColor_Specified, specifiedColor=<rgba 0.0513 0.0953 0.159 1>}
      uVRegion {bIsValid=false, max=<0 0>, min=<0 0>}
     }
     inactiveHighlightedBrush {
      bIsDynamicallyLoaded=false
      drawAs=Image
      imageType=NoImage
      mirroring=NoMirror
      resourceName=None
      resourceObject=null
      tiling=NoTile
      imageSize=<0 0>
      margin=<ltrb 0 0 0 0>
      outlineSettings {
       bUseBrushTransparency=false
       roundingType=HalfHeightRadius
       width=0
       color {colorUseRule=UseColor_Specified, specifiedColor=<rgba 0 0 0 0>}
       cornerRadii=<0 0 0 0>
      }
      tintColor {colorUseRule=UseColor_Specified, specifiedColor=<rgba 0.0252 0.0319 0.0423 1>}
      uVRegion {bIsValid=false, max=<0 0>, min=<0 0>}
     }
     inactiveHoveredBrush {
      bIsDynamicallyLoaded=false
      drawAs=Image
      imageType=NoImage
      mirroring=NoMirror
      resourceName=None
      resourceObject=null
      tiling=NoTile
      imageSize=<0 0>
      margin=<ltrb 0 0 0 0>
      outlineSettings {
       bUseBrushTransparency=false
       roundingType=HalfHeightRadius
       width=0
       color {colorUseRule=UseColor_Specified, specifiedColor=<rgba 0 0 0 0>}
       cornerRadii=<0 0 0 0>
      }
      tintColor {colorUseRule=UseColor_Specified, specifiedColor=<rgba 0.0513 0.0953 0.159 1>}
      uVRegion {bIsValid=false, max=<0 0>, min=<0 0>}
     }
     oddRowBackgroundBrush {
      bIsDynamicallyLoaded=false
      drawAs=Image
      imageType=NoImage
      mirroring=NoMirror
      resourceName=None
      resourceObject=null
      tiling=NoTile
      imageSize=<0 0>
      margin=<ltrb 0 0 0 0>
      outlineSettings {
       bUseBrushTransparency=false
       roundingType=HalfHeightRadius
       width=0
       color {colorUseRule=UseColor_Specified, specifiedColor=<rgba 0 0 0 0>}
       cornerRadii=<0 0 0 0>
      }
      tintColor {colorUseRule=UseColor_Specified, specifiedColor=<rgba 0.0103 0.0103 0.0103 1>}
      uVRegion {bIsValid=false, max=<0 0>, min=<0 0>}
     }
     oddRowBackgroundHoveredBrush {
      bIsDynamicallyLoaded=false
      drawAs=Image
      imageType=NoImage
      mirroring=NoMirror
      resourceName=None
      resourceObject=null
      tiling=NoTile
      imageSize=<0 0>
      margin=<ltrb 0 0 0 0>
      outlineSettings {
       bUseBrushTransparency=false
       roundingType=HalfHeightRadius
       width=0
       color {colorUseRule=UseColor_Specified, specifiedColor=<rgba 0 0 0 0>}
       cornerRadii=<0 0 0 0>
      }
      tintColor {colorUseRule=UseColor_Specified, specifiedColor=<rgba 0.0176 0.0176 0.0176 1>}
      uVRegion {bIsValid=false, max=<0 0>, min=<0 0>}
     }
     parentRowBackgroundBrush {
      bIsDynamicallyLoaded=false
      drawAs=Image
      imageType=NoImage
      mirroring=NoMirror
      resourceName=None
      resourceObject=null
      tiling=NoTile
      imageSize=<32 32>
      margin=<ltrb 0 0 0 0>
      outlineSettings {
       bUseBrushTransparency=false
       roundingType=HalfHeightRadius
       width=0
       color {colorUseRule=UseColor_Specified, specifiedColor=<rgba 0 0 0 0>}
       cornerRadii=<0 0 0 0>
      }
      tintColor {colorUseRule=UseColor_Specified, specifiedColor=<rgba 1 1 1 1>}
      uVRegion {bIsValid=false, max=<0 0>, min=<0 0>}
     }
     parentRowBackgroundHoveredBrush {
      bIsDynamicallyLoaded=false
      drawAs=Image
      imageType=NoImage
      mirroring=NoMirror
      resourceName=None
      resourceObject=null
      tiling=NoTile
      imageSize=<32 32>
      margin=<ltrb 0 0 0 0>
      outlineSettings {
       bUseBrushTransparency=false
       roundingType=HalfHeightRadius
       width=0
       color {colorUseRule=UseColor_Specified, specifiedColor=<rgba 0 0 0 0>}
       cornerRadii=<0 0 0 0>
      }
      tintColor {colorUseRule=UseColor_Specified, specifiedColor=<rgba 1 1 1 1>}
      uVRegion {bIsValid=false, max=<0 0>, min=<0 0>}
     }
     selectedTextColor {colorUseRule=UseColor_Specified, specifiedColor=<rgba 1 1 1 1>}
     selectorFocusedBrush {
      bIsDynamicallyLoaded=false
      drawAs=Border
      imageType=FullColor
      mirroring=NoMirror
      resourceName=../../../Engine/Content/Slate/Common/Selector.png
      resourceObject=null
      tiling=Both
      imageSize=<0 0>
      margin=<ltrb 0.25 0.25 0.25 0.25>
      outlineSettings {
       bUseBrushTransparency=false
       roundingType=HalfHeightRadius
       width=0
       color {colorUseRule=UseColor_Specified, specifiedColor=<rgba 0 0 0 0>}
       cornerRadii=<0 0 0 0>
      }
      tintColor {colorUseRule=UseColor_Specified, specifiedColor=<rgba 0.701 0.225 0.003 1>}
      uVRegion {bIsValid=false, max=<0 0>, min=<0 0>}
     }
     textColor {colorUseRule=UseColor_Specified, specifiedColor=<rgba 0.5271 0.5271 0.5271 1>}
    }
    widgetStyle {
     comboButtonStyle {
      downArrowAlign=VAlign_Center
      buttonStyle {
       clickedSlateSound {resourceObject=null}
       disabled {
        bIsDynamicallyLoaded=false
        drawAs=RoundedBox
        imageType=NoImage
        mirroring=NoMirror
        resourceName=None
        resourceObject=null
        tiling=NoTile
        imageSize=<0 0>
        margin=<ltrb 0 0 0 0>
        outlineSettings {
         bUseBrushTransparency=false
         roundingType=FixedRadius
         width=1
         color {colorUseRule=UseColor_ColorTable, specifiedColor=<rgba 0 0 0 0>}
         cornerRadii=<4 4 4 4>
        }
        tintColor {colorUseRule=UseColor_Specified, specifiedColor=<rgba 0.0395 0.0395 0.0395 1>}
        uVRegion {bIsValid=false, max=<0 0>, min=<0 0>}
       }
       disabledForeground {colorUseRule=UseColor_ColorTable, specifiedColor=<rgba 0 0 0 0>}
       hovered {
        bIsDynamicallyLoaded=false
        drawAs=RoundedBox
        imageType=NoImage
        mirroring=NoMirror
        resourceName=None
        resourceObject=null
        tiling=NoTile
        imageSize=<0 0>
        margin=<ltrb 0 0 0 0>
        outlineSettings {
         bUseBrushTransparency=false
         roundingType=FixedRadius
         width=1
         color {colorUseRule=UseColor_ColorTable, specifiedColor=<rgba 0 0 0 0>}
         cornerRadii=<4 4 4 4>
        }
        tintColor {colorUseRule=UseColor_Specified, specifiedColor=<rgba 0.0048 0.0048 0.0048 1>}
        uVRegion {bIsValid=false, max=<0 0>, min=<0 0>}
       }
       hoveredForeground {colorUseRule=UseColor_ColorTable, specifiedColor=<rgba 0 0 0 0>}
       hoveredSlateSound {resourceObject=null}
       normal {
        bIsDynamicallyLoaded=false
        drawAs=RoundedBox
        imageType=NoImage
        mirroring=NoMirror
        resourceName=None
        resourceObject=null
        tiling=NoTile
        imageSize=<0 0>
        margin=<ltrb 0 0 0 0>
        outlineSettings {
         bUseBrushTransparency=false
         roundingType=FixedRadius
         width=1
         color {colorUseRule=UseColor_ColorTable, specifiedColor=<rgba 0 0 0 0>}
         cornerRadii=<4 4 4 4>
        }
        tintColor {colorUseRule=UseColor_Specified, specifiedColor=<rgba 0.0048 0.0048 0.0048 1>}
        uVRegion {bIsValid=false, max=<0 0>, min=<0 0>}
       }
       normalForeground {colorUseRule=UseColor_ColorTable, specifiedColor=<rgba 0 0 0 0>}
       normalPadding=<ltrb 8 1 8 1>
       pressed {
        bIsDynamicallyLoaded=false
        drawAs=RoundedBox
        imageType=NoImage
        mirroring=NoMirror
        resourceName=None
        resourceObject=null
        tiling=NoTile
        imageSize=<0 0>
        margin=<ltrb 0 0 0 0>
        outlineSettings {
         bUseBrushTransparency=false
         roundingType=FixedRadius
         width=1
         color {colorUseRule=UseColor_ColorTable, specifiedColor=<rgba 0 0 0 0>}
         cornerRadii=<4 4 4 4>
        }
        tintColor {colorUseRule=UseColor_Specified, specifiedColor=<rgba 0.0048 0.0048 0.0048 1>}
        uVRegion {bIsValid=false, max=<0 0>, min=<0 0>}
       }
       pressedForeground {colorUseRule=UseColor_ColorTable, specifiedColor=<rgba 0 0 0 0>}
       pressedPadding=<ltrb 8 2 8 0>
       pressedSlateSound {resourceObject=null}
      }
      contentPadding=<ltrb 0 0 0 0>
      downArrowImage {
       bIsDynamicallyLoaded=false
       drawAs=Image
       imageType=Vector
       mirroring=NoMirror
       resourceName=../../../Engine/Content/Slate/Starship/CoreWidgets/ComboBox/wide-chevron-down.svg
       resourceObject=null
       tiling=NoTile
       imageSize=<16 16>
       margin=<ltrb 0 0 0 0>
       outlineSettings {
        bUseBrushTransparency=false
        roundingType=HalfHeightRadius
        width=0
        color {colorUseRule=UseColor_Specified, specifiedColor=<rgba 0 0 0 0>}
        cornerRadii=<0 0 0 0>
       }
       tintColor {colorUseRule=UseColor_Specified, specifiedColor=<rgba 1 1 1 1>}
       uVRegion {bIsValid=false, max=<0 0>, min=<0 0>}
      }
      downArrowPadding=<ltrb 2 2 2 2>
      menuBorderBrush {
       bIsDynamicallyLoaded=false
       drawAs=RoundedBox
       imageType=NoImage
       mirroring=NoMirror
       resourceName=None
       resourceObject=null
       tiling=NoTile
       imageSize=<0 0>
       margin=<ltrb 0 0 0 0>
       outlineSettings {
        bUseBrushTransparency=false
        roundingType=FixedRadius
        width=1
        color {colorUseRule=UseColor_ColorTable, specifiedColor=<rgba 0 0 0 0>}
        cornerRadii=<0 0 0 0>
       }
       tintColor {colorUseRule=UseColor_Specified, specifiedColor=<rgba 0.0395 0.0395 0.0395 1>}
       uVRegion {bIsValid=false, max=<0 0>, min=<0 0>}
      }
      menuBorderPadding=<ltrb 0 0 0 0>
      shadowColorAndOpacity=<rgba 0 0 0 1>
      shadowOffset=<0 0>
     }
     contentPadding=<ltrb 0 0 0 0>
     menuRowPadding=<ltrb 8 3 8 3>
     pressedSlateSound {resourceObject=null}
     selectionChangeSlateSound {resourceObject=null}
    }
   }
   slot {class=VerticalBoxSlot}
  }
  NameHorizontalBox {
   widgetClassPath=&HorizontalBox
   parent=NameVerticalBox
   widget=NameHorizontalBox
   slot {class=VerticalBoxSlot, properties {padding=<ltrb 0 2 0 2>}}
  }
  SizeBox_4 {
   widgetClassPath=&SizeBox
   parent=NameHorizontalBox
   widget=SizeBox_4
   properties {bOverride_MinDesiredWidth=true, maxDesiredWidth=100, minDesiredWidth=100}
   slot {class=HorizontalBoxSlot, properties {verticalAlignment=VAlign_Center}}
  }
  NameTextBlock {
   widgetClassPath=&TextBlock
   bIsVariable=true
   parent=SizeBox_4
   widget=NameTextBlock
   properties {
    text=Animation Name
    toolTipText=What Am I doing?
    colorAndOpacity {colorUseRule=UseColor_Specified, specifiedColor=<rgba 0.5255 0.5255 0.5255 1>}
    font {
     bForceMonospaced=false
     bMaterialIsStencil=false
     fontMaterial=null
     fontObject=&Roboto
     letterSpacing=0
     monospacedWidth=1
     size=10
     skewAmount=0
     typefaceFontName=Bold
     outlineSettings {
      bApplyOutlineToDropShadows=false
      bMiteredCorners=false
      bSeparateFillAlpha=false
      outlineMaterial=null
      outlineSize=0
      outlineColor=<rgba 0 0 0 1>
     }
    }
   }
   slot {class=SizeBoxSlot, properties {horizontalAlignment=HAlign_Right}}
  }
  NameTextBox {
   widgetClassPath=&EditableTextBox
   bIsVariable=true
   parent=NameHorizontalBox
   widget=NameTextBox
   properties {
    hintText="What am I doing? (spacebar to skip)"
    toolTipText=What Am I doing?
    widgetStyle {
     backgroundColor {colorUseRule=UseColor_Specified, specifiedColor=<rgba 0.0048 0.0048 0.0048 1>}
     backgroundImageFocused {
      bIsDynamicallyLoaded=false
      drawAs=RoundedBox
      imageType=NoImage
      mirroring=NoMirror
      resourceName=None
      resourceObject=null
      tiling=NoTile
      imageSize=<0 0>
      margin=<ltrb 0 0 0 0>
      outlineSettings {
       bUseBrushTransparency=false
       roundingType=FixedRadius
       width=1
       color {colorUseRule=UseColor_ColorTable, specifiedColor=<rgba 0 0 0 0>}
       cornerRadii=<4 4 4 4>
      }
      tintColor {colorUseRule=UseColor_Specified, specifiedColor=<rgba 0.0048 0.0048 0.0048 1>}
      uVRegion {bIsValid=false, max=<0 0>, min=<0 0>}
     }
     backgroundImageHovered {
      bIsDynamicallyLoaded=false
      drawAs=RoundedBox
      imageType=NoImage
      mirroring=NoMirror
      resourceName=None
      resourceObject=null
      tiling=NoTile
      imageSize=<0 0>
      margin=<ltrb 0 0 0 0>
      outlineSettings {
       bUseBrushTransparency=false
       roundingType=FixedRadius
       width=1
       color {colorUseRule=UseColor_ColorTable, specifiedColor=<rgba 0 0 0 0>}
       cornerRadii=<4 4 4 4>
      }
      tintColor {colorUseRule=UseColor_Specified, specifiedColor=<rgba 0.0048 0.0048 0.0048 1>}
      uVRegion {bIsValid=false, max=<0 0>, min=<0 0>}
     }
     backgroundImageNormal {
      bIsDynamicallyLoaded=false
      drawAs=RoundedBox
      imageType=NoImage
      mirroring=NoMirror
      resourceName=None
      resourceObject=null
      tiling=NoTile
      imageSize=<0 0>
      margin=<ltrb 0 0 0 0>
      outlineSettings {
       bUseBrushTransparency=false
       roundingType=FixedRadius
       width=1
       color {colorUseRule=UseColor_ColorTable, specifiedColor=<rgba 0 0 0 0>}
       cornerRadii=<4 4 4 4>
      }
      tintColor {colorUseRule=UseColor_Specified, specifiedColor=<rgba 0.0048 0.0048 0.0048 1>}
      uVRegion {bIsValid=false, max=<0 0>, min=<0 0>}
     }
     backgroundImageReadOnly {
      bIsDynamicallyLoaded=false
      drawAs=RoundedBox
      imageType=NoImage
      mirroring=NoMirror
      resourceName=None
      resourceObject=null
      tiling=NoTile
      imageSize=<0 0>
      margin=<ltrb 0 0 0 0>
      outlineSettings {
       bUseBrushTransparency=false
       roundingType=FixedRadius
       width=1
       color {colorUseRule=UseColor_ColorTable, specifiedColor=<rgba 0 0 0 0>}
       cornerRadii=<4 4 4 4>
      }
      tintColor {colorUseRule=UseColor_Specified, specifiedColor=<rgba 0.0284 0.0284 0.0284 1>}
      uVRegion {bIsValid=false, max=<0 0>, min=<0 0>}
     }
     focusedForegroundColor {colorUseRule=UseColor_Specified, specifiedColor=<rgba 1 1 1 1>}
     foregroundColor {colorUseRule=UseColor_Specified, specifiedColor=<rgba 0.5271 0.5271 0.5271 1>}
     hScrollBarPadding=<ltrb 0 0 0 0>
     padding=<ltrb 8 3 8 4>
     readOnlyForegroundColor {colorUseRule=UseColor_Foreground, specifiedColor=<rgba 1 0 1 1>}
     scrollBarStyle {
      thickness=8
      draggedThumbImage {
       bIsDynamicallyLoaded=false
       drawAs=RoundedBox
       imageType=NoImage
       mirroring=NoMirror
       resourceName=None
       resourceObject=null
       tiling=NoTile
       imageSize=<0 0>
       margin=<ltrb 0 0 0 0>
       outlineSettings {
        bUseBrushTransparency=false
        roundingType=FixedRadius
        width=0
        color {colorUseRule=UseColor_Specified, specifiedColor=<rgba 0 0 0 0>}
        cornerRadii=<4 4 4 4>
       }
       tintColor {colorUseRule=UseColor_Specified, specifiedColor=<rgba 0.2159 0.2159 0.2159 1>}
       uVRegion {bIsValid=false, max=<0 0>, min=<0 0>}
      }
      horizontalBackgroundImage {
       bIsDynamicallyLoaded=false
       drawAs=NoDrawType
       imageType=NoImage
       mirroring=NoMirror
       resourceName=None
       resourceObject=null
       tiling=NoTile
       imageSize=<0 0>
       margin=<ltrb 0 0 0 0>
       outlineSettings {
        bUseBrushTransparency=false
        roundingType=HalfHeightRadius
        width=0
        color {colorUseRule=UseColor_Specified, specifiedColor=<rgba 0 0 0 0>}
        cornerRadii=<0 0 0 0>
       }
       tintColor {colorUseRule=UseColor_Specified, specifiedColor=<rgba 1 1 1 1>}
       uVRegion {bIsValid=false, max=<0 0>, min=<0 0>}
      }
      horizontalBottomSlotImage {
       bIsDynamicallyLoaded=false
       drawAs=NoDrawType
       imageType=NoImage
       mirroring=NoMirror
       resourceName=None
       resourceObject=null
       tiling=NoTile
       imageSize=<0 0>
       margin=<ltrb 0 0 0 0>
       outlineSettings {
        bUseBrushTransparency=false
        roundingType=HalfHeightRadius
        width=0
        color {colorUseRule=UseColor_Specified, specifiedColor=<rgba 0 0 0 0>}
        cornerRadii=<0 0 0 0>
       }
       tintColor {colorUseRule=UseColor_Specified, specifiedColor=<rgba 1 1 1 1>}
       uVRegion {bIsValid=false, max=<0 0>, min=<0 0>}
      }
      horizontalTopSlotImage {
       bIsDynamicallyLoaded=false
       drawAs=NoDrawType
       imageType=NoImage
       mirroring=NoMirror
       resourceName=None
       resourceObject=null
       tiling=NoTile
       imageSize=<0 0>
       margin=<ltrb 0 0 0 0>
       outlineSettings {
        bUseBrushTransparency=false
        roundingType=HalfHeightRadius
        width=0
        color {colorUseRule=UseColor_Specified, specifiedColor=<rgba 0 0 0 0>}
        cornerRadii=<0 0 0 0>
       }
       tintColor {colorUseRule=UseColor_Specified, specifiedColor=<rgba 1 1 1 1>}
       uVRegion {bIsValid=false, max=<0 0>, min=<0 0>}
      }
      hoveredThumbImage {
       bIsDynamicallyLoaded=false
       drawAs=RoundedBox
       imageType=NoImage
       mirroring=NoMirror
       resourceName=None
       resourceObject=null
       tiling=NoTile
       imageSize=<0 0>
       margin=<ltrb 0 0 0 0>
       outlineSettings {
        bUseBrushTransparency=false
        roundingType=FixedRadius
        width=0
        color {colorUseRule=UseColor_Specified, specifiedColor=<rgba 0 0 0 0>}
        cornerRadii=<4 4 4 4>
       }
       tintColor {colorUseRule=UseColor_Specified, specifiedColor=<rgba 0.2159 0.2159 0.2159 1>}
       uVRegion {bIsValid=false, max=<0 0>, min=<0 0>}
      }
      normalThumbImage {
       bIsDynamicallyLoaded=false
       drawAs=RoundedBox
       imageType=NoImage
       mirroring=NoMirror
       resourceName=None
       resourceObject=null
       tiling=NoTile
       imageSize=<0 0>
       margin=<ltrb 0 0 0 0>
       outlineSettings {
        bUseBrushTransparency=false
        roundingType=FixedRadius
        width=0
        color {colorUseRule=UseColor_Specified, specifiedColor=<rgba 0 0 0 0>}
        cornerRadii=<4 4 4 4>
       }
       tintColor {colorUseRule=UseColor_Specified, specifiedColor=<rgba 0.0953 0.0953 0.0953 1>}
       uVRegion {bIsValid=false, max=<0 0>, min=<0 0>}
      }
      verticalBackgroundImage {
       bIsDynamicallyLoaded=false
       drawAs=NoDrawType
       imageType=NoImage
       mirroring=NoMirror
       resourceName=None
       resourceObject=null
       tiling=NoTile
       imageSize=<0 0>
       margin=<ltrb 0 0 0 0>
       outlineSettings {
        bUseBrushTransparency=false
        roundingType=HalfHeightRadius
        width=0
        color {colorUseRule=UseColor_Specified, specifiedColor=<rgba 0 0 0 0>}
        cornerRadii=<0 0 0 0>
       }
       tintColor {colorUseRule=UseColor_Specified, specifiedColor=<rgba 1 1 1 1>}
       uVRegion {bIsValid=false, max=<0 0>, min=<0 0>}
      }
      verticalBottomSlotImage {
       bIsDynamicallyLoaded=false
       drawAs=NoDrawType
       imageType=NoImage
       mirroring=NoMirror
       resourceName=None
       resourceObject=null
       tiling=NoTile
       imageSize=<0 0>
       margin=<ltrb 0 0 0 0>
       outlineSettings {
        bUseBrushTransparency=false
        roundingType=HalfHeightRadius
        width=0
        color {colorUseRule=UseColor_Specified, specifiedColor=<rgba 0 0 0 0>}
        cornerRadii=<0 0 0 0>
       }
       tintColor {colorUseRule=UseColor_Specified, specifiedColor=<rgba 1 1 1 1>}
       uVRegion {bIsValid=false, max=<0 0>, min=<0 0>}
      }
      verticalTopSlotImage {
       bIsDynamicallyLoaded=false
       drawAs=NoDrawType
       imageType=NoImage
       mirroring=NoMirror
       resourceName=None
       resourceObject=null
       tiling=NoTile
       imageSize=<0 0>
       margin=<ltrb 0 0 0 0>
       outlineSettings {
        bUseBrushTransparency=false
        roundingType=HalfHeightRadius
        width=0
        color {colorUseRule=UseColor_Specified, specifiedColor=<rgba 0 0 0 0>}
        cornerRadii=<0 0 0 0>
       }
       tintColor {colorUseRule=UseColor_Specified, specifiedColor=<rgba 1 1 1 1>}
       uVRegion {bIsValid=false, max=<0 0>, min=<0 0>}
      }
     }
     textStyle {
      overflowPolicy=Clip
      transformPolicy=None
      colorAndOpacity {colorUseRule=UseColor_Foreground, specifiedColor=<rgba 1 0 1 1>}
      font {
       bForceMonospaced=false
       bMaterialIsStencil=false
       fontMaterial=null
       fontObject=null
       letterSpacing=0
       monospacedWidth=1
       size=9
       skewAmount=0
       typefaceFontName=Regular
       outlineSettings {
        bApplyOutlineToDropShadows=false
        bMiteredCorners=false
        bSeparateFillAlpha=false
        outlineMaterial=null
        outlineSize=0
        outlineColor=<rgba 0 0 0 1>
       }
      }
      highlightColor {colorUseRule=UseColor_ColorTable, specifiedColor=<rgba 0 0 0 0>}
      highlightShape {
       bIsDynamicallyLoaded=false
       drawAs=Image
       imageType=NoImage
       mirroring=NoMirror
       resourceName=None
       resourceObject=null
       tiling=NoTile
       imageSize=<0 0>
       margin=<ltrb 0 0 0 0>
       outlineSettings {
        bUseBrushTransparency=false
        roundingType=HalfHeightRadius
        width=0
        color {colorUseRule=UseColor_Specified, specifiedColor=<rgba 0 0 0 0>}
        cornerRadii=<0 0 0 0>
       }
       tintColor {colorUseRule=UseColor_ColorTable, specifiedColor=<rgba 0 0 0 0>}
       uVRegion {bIsValid=false, max=<0 0>, min=<0 0>}
      }
      selectedBackgroundColor {colorUseRule=UseColor_Specified, specifiedColor=<rgba 0 0.162 0.7454 1>}
      shadowColorAndOpacity=<rgba 0 0 0 1>
      shadowOffset=<0 0>
      strikeBrush {
       bIsDynamicallyLoaded=false
       drawAs=Image
       imageType=NoImage
       mirroring=NoMirror
       resourceName=None
       resourceObject=null
       tiling=NoTile
       imageSize=<32 32>
       margin=<ltrb 0 0 0 0>
       outlineSettings {
        bUseBrushTransparency=false
        roundingType=HalfHeightRadius
        width=0
        color {colorUseRule=UseColor_Specified, specifiedColor=<rgba 0 0 0 0>}
        cornerRadii=<0 0 0 0>
       }
       tintColor {colorUseRule=UseColor_Specified, specifiedColor=<rgba 1 1 1 1>}
       uVRegion {bIsValid=false, max=<0 0>, min=<0 0>}
      }
      underlineBrush {
       bIsDynamicallyLoaded=false
       drawAs=Image
       imageType=NoImage
       mirroring=NoMirror
       resourceName=None
       resourceObject=null
       tiling=NoTile
       imageSize=<32 32>
       margin=<ltrb 0 0 0 0>
       outlineSettings {
        bUseBrushTransparency=false
        roundingType=HalfHeightRadius
        width=0
        color {colorUseRule=UseColor_Specified, specifiedColor=<rgba 0 0 0 0>}
        cornerRadii=<0 0 0 0>
       }
       tintColor {colorUseRule=UseColor_Specified, specifiedColor=<rgba 1 1 1 1>}
       uVRegion {bIsValid=false, max=<0 0>, min=<0 0>}
      }
     }
     vScrollBarPadding=<ltrb 0 0 0 0>
    }
   }
   slot {class=HorizontalBoxSlot, properties {padding=<ltrb 5 0 0 0>, size {sizeRule=Fill, value=1}}}
  }
  DescriptorHorizontalBox {
   widgetClassPath=&HorizontalBox
   bIsVariable=true
   parent=NameVerticalBox
   widget=DescriptorHorizontalBox
   slot {class=VerticalBoxSlot, properties {padding=<ltrb 0 2 0 2>}}
  }
  SizeBox {
   widgetClassPath=&SizeBox
   parent=DescriptorHorizontalBox
   widget=SizeBox
   properties {bOverride_MinDesiredWidth=true, minDesiredWidth=100}
   slot {class=HorizontalBoxSlot, properties {verticalAlignment=VAlign_Center}}
  }
  DescriptorTextBlock {
   widgetClassPath=&TextBlock
   bIsVariable=true
   parent=SizeBox
   widget=DescriptorTextBlock
   properties {
    text=Descriptor
    toolTipText="Technically speaking, what is this File for/How Is it used?\r\nHit the Pencil to type in your own description if missing from the list"
    colorAndOpacity {colorUseRule=UseColor_Specified, specifiedColor=<rgba 0 0.5255 0.3373 1>}
    font {
     bForceMonospaced=false
     bMaterialIsStencil=false
     fontMaterial=null
     fontObject=&Roboto
     letterSpacing=0
     monospacedWidth=1
     size=10
     skewAmount=0
     typefaceFontName=Bold
     outlineSettings {
      bApplyOutlineToDropShadows=false
      bMiteredCorners=false
      bSeparateFillAlpha=false
      outlineMaterial=null
      outlineSize=0
      outlineColor=<rgba 0 0 0 1>
     }
    }
   }
   slot {class=SizeBoxSlot, properties {horizontalAlignment=HAlign_Right}}
  }
  VerticalBox_1 {
   widgetClassPath=&VerticalBox
   parent=DescriptorHorizontalBox
   widget=VerticalBox_1
   slot {class=HorizontalBoxSlot, properties {padding=<ltrb 5 0 0 0>, size {sizeRule=Fill, value=1}}}
  }
  DescriptorCustomEntryComboBox {
   widgetClassPath=&EUW_CustomEntryComboBox
   bIsVariable=true
   parent=VerticalBox_1
   widget=DescriptorCustomEntryComboBox
   properties {toolTipText="Technically speaking, what is this File for/How Is it used?\r\nHit the Pencil to type in your own description if missing from the list"}
   slot {class=VerticalBoxSlot}
  }
  DirectionHorizontalBox {
   widgetClassPath=&HorizontalBox
   bIsVariable=true
   parent=NameVerticalBox
   widget=DirectionHorizontalBox
   slot {class=VerticalBoxSlot, properties {padding=<ltrb 0 2 0 2>}}
  }
  SizeBox_8 {
   widgetClassPath=&SizeBox
   parent=DirectionHorizontalBox
   widget=SizeBox_8
   properties {bOverride_MinDesiredWidth=true, minDesiredWidth=100}
   slot {class=HorizontalBoxSlot, properties {verticalAlignment=VAlign_Top, padding=<ltrb 0 5 0 0>}}
  }
  DirectionTextBlock {
   widgetClassPath=&TextBlock
   bIsVariable=true
   parent=SizeBox_8
   widget=DirectionTextBlock
   properties {
    text=Direction
    toolTipText=Which way am I going?
    colorAndOpacity {colorUseRule=UseColor_Specified, specifiedColor=<rgba 0 0.5255 0.3373 1>}
    font {
     bForceMonospaced=false
     bMaterialIsStencil=false
     fontMaterial=null
     fontObject=&Roboto
     letterSpacing=0
     monospacedWidth=1
     size=10
     skewAmount=0
     typefaceFontName=Bold
     outlineSettings {
      bApplyOutlineToDropShadows=false
      bMiteredCorners=false
      bSeparateFillAlpha=false
      outlineMaterial=null
      outlineSize=0
      outlineColor=<rgba 0 0 0 1>
     }
    }
   }
   slot {class=SizeBoxSlot, properties {horizontalAlignment=HAlign_Right}}
  }
  VerticalBox_2 {
   widgetClassPath=&VerticalBox
   parent=DirectionHorizontalBox
   widget=VerticalBox_2
   slot {class=HorizontalBoxSlot, properties {padding=<ltrb 5 0 0 0>, size {sizeRule=Fill, value=1}}}
  }
  DirectionComboBox {
   widgetClassPath=&SearchableStringComboBox
   bIsVariable=true
   parent=VerticalBox_2
   widget=DirectionComboBox
   properties {
    font {
     bForceMonospaced=false
     bMaterialIsStencil=false
     fontMaterial=null
     fontObject=&Roboto
     letterSpacing=0
     monospacedWidth=1
     size=10
     skewAmount=0
     typefaceFontName=Regular
     outlineSettings {
      bApplyOutlineToDropShadows=false
      bMiteredCorners=false
      bSeparateFillAlpha=false
      outlineMaterial=null
      outlineSize=0
      outlineColor=<rgba 0 0 0 1>
     }
    }
    foregroundColor {colorUseRule=UseColor_Specified, specifiedColor=<rgba 0.5271 0.5271 0.5271 1>}
    itemStyle {
     bUseParentRowBrush=false
     activeBrush {
      bIsDynamicallyLoaded=false
      drawAs=Image
      imageType=NoImage
      mirroring=NoMirror
      resourceName=None
      resourceObject=null
      tiling=NoTile
      imageSize=<0 0>
      margin=<ltrb 0 0 0 0>
      outlineSettings {
       bUseBrushTransparency=false
       roundingType=HalfHeightRadius
       width=0
       color {colorUseRule=UseColor_Specified, specifiedColor=<rgba 0 0 0 0>}
       cornerRadii=<0 0 0 0>
      }
      tintColor {colorUseRule=UseColor_Specified, specifiedColor=<rgba 0 0.162 0.7454 1>}
      uVRegion {bIsValid=false, max=<0 0>, min=<0 0>}
     }
     activeHighlightedBrush {
      bIsDynamicallyLoaded=false
      drawAs=Image
      imageType=NoImage
      mirroring=NoMirror
      resourceName=None
      resourceObject=null
      tiling=NoTile
      imageSize=<0 0>
      margin=<ltrb 0 0 0 0>
      outlineSettings {
       bUseBrushTransparency=false
       roundingType=HalfHeightRadius
       width=0
       color {colorUseRule=UseColor_Specified, specifiedColor=<rgba 0 0 0 0>}
       cornerRadii=<0 0 0 0>
      }
      tintColor {colorUseRule=UseColor_Specified, specifiedColor=<rgba 0.0252 0.0319 0.0423 1>}
      uVRegion {bIsValid=false, max=<0 0>, min=<0 0>}
     }
     activeHoveredBrush {
      bIsDynamicallyLoaded=false
      drawAs=Image
      imageType=NoImage
      mirroring=NoMirror
      resourceName=None
      resourceObject=null
      tiling=NoTile
      imageSize=<0 0>
      margin=<ltrb 0 0 0 0>
      outlineSettings {
       bUseBrushTransparency=false
       roundingType=HalfHeightRadius
       width=0
       color {colorUseRule=UseColor_Specified, specifiedColor=<rgba 0 0 0 0>}
       cornerRadii=<0 0 0 0>
      }
      tintColor {colorUseRule=UseColor_Specified, specifiedColor=<rgba 0 0.162 0.7454 1>}
      uVRegion {bIsValid=false, max=<0 0>, min=<0 0>}
     }
     dropIndicator_Above {
      bIsDynamicallyLoaded=false
      drawAs=Box
      imageType=FullColor
      mirroring=NoMirror
      resourceName=../../../Engine/Content/Slate/Common/DropZoneIndicator_Above.png
      resourceObject=null
      tiling=NoTile
      imageSize=<0 0>
      margin=<ltrb 0.625 0.625 0 0>
      outlineSettings {
       bUseBrushTransparency=false
       roundingType=HalfHeightRadius
       width=0
       color {colorUseRule=UseColor_Specified, specifiedColor=<rgba 0 0 0 0>}
       cornerRadii=<0 0 0 0>
      }
      tintColor {colorUseRule=UseColor_Specified, specifiedColor=<rgba 0.0091 0.3515 0.9647 1>}
      uVRegion {bIsValid=false, max=<0 0>, min=<0 0>}
     }
     dropIndicator_Below {
      bIsDynamicallyLoaded=false
      drawAs=Box
      imageType=FullColor
      mirroring=NoMirror
      resourceName=../../../Engine/Content/Slate/Common/DropZoneIndicator_Below.png
      resourceObject=null
      tiling=NoTile
      imageSize=<0 0>
      margin=<ltrb 0.625 0 0 0.625>
      outlineSettings {
       bUseBrushTransparency=false
       roundingType=HalfHeightRadius
       width=0
       color {colorUseRule=UseColor_Specified, specifiedColor=<rgba 0 0 0 0>}
       cornerRadii=<0 0 0 0>
      }
      tintColor {colorUseRule=UseColor_Specified, specifiedColor=<rgba 0.0091 0.3515 0.9647 1>}
      uVRegion {bIsValid=false, max=<0 0>, min=<0 0>}
     }
     dropIndicator_Onto {
      bIsDynamicallyLoaded=false
      drawAs=Box
      imageType=FullColor
      mirroring=NoMirror
      resourceName=../../../Engine/Content/Slate/Common/DropZoneIndicator_Onto.png
      resourceObject=null
      tiling=NoTile
      imageSize=<0 0>
      margin=<ltrb 0.25 0.25 0.25 0.25>
      outlineSettings {
       bUseBrushTransparency=false
       roundingType=HalfHeightRadius
       width=0
       color {colorUseRule=UseColor_Specified, specifiedColor=<rgba 0 0 0 0>}
       cornerRadii=<0 0 0 0>
      }
      tintColor {colorUseRule=UseColor_Specified, specifiedColor=<rgba 0.0091 0.3515 0.9647 1>}
      uVRegion {bIsValid=false, max=<0 0>, min=<0 0>}
     }
     evenRowBackgroundBrush {
      bIsDynamicallyLoaded=false
      drawAs=Image
      imageType=NoImage
      mirroring=NoMirror
      resourceName=None
      resourceObject=null
      tiling=NoTile
      imageSize=<0 0>
      margin=<ltrb 0 0 0 0>
      outlineSettings {
       bUseBrushTransparency=false
       roundingType=HalfHeightRadius
       width=0
       color {colorUseRule=UseColor_Specified, specifiedColor=<rgba 0 0 0 0>}
       cornerRadii=<0 0 0 0>
      }
      tintColor {colorUseRule=UseColor_Specified, specifiedColor=<rgba 0.0103 0.0103 0.0103 1>}
      uVRegion {bIsValid=false, max=<0 0>, min=<0 0>}
     }
     evenRowBackgroundHoveredBrush {
      bIsDynamicallyLoaded=false
      drawAs=Image
      imageType=NoImage
      mirroring=NoMirror
      resourceName=None
      resourceObject=null
      tiling=NoTile
      imageSize=<0 0>
      margin=<ltrb 0 0 0 0>
      outlineSettings {
       bUseBrushTransparency=false
       roundingType=HalfHeightRadius
       width=0
       color {colorUseRule=UseColor_Specified, specifiedColor=<rgba 0 0 0 0>}
       cornerRadii=<0 0 0 0>
      }
      tintColor {colorUseRule=UseColor_Specified, specifiedColor=<rgba 0.0176 0.0176 0.0176 1>}
      uVRegion {bIsValid=false, max=<0 0>, min=<0 0>}
     }
     inactiveBrush {
      bIsDynamicallyLoaded=false
      drawAs=Image
      imageType=NoImage
      mirroring=NoMirror
      resourceName=None
      resourceObject=null
      tiling=NoTile
      imageSize=<0 0>
      margin=<ltrb 0 0 0 0>
      outlineSettings {
       bUseBrushTransparency=false
       roundingType=HalfHeightRadius
       width=0
       color {colorUseRule=UseColor_Specified, specifiedColor=<rgba 0 0 0 0>}
       cornerRadii=<0 0 0 0>
      }
      tintColor {colorUseRule=UseColor_Specified, specifiedColor=<rgba 0.0513 0.0953 0.159 1>}
      uVRegion {bIsValid=false, max=<0 0>, min=<0 0>}
     }
     inactiveHighlightedBrush {
      bIsDynamicallyLoaded=false
      drawAs=Image
      imageType=NoImage
      mirroring=NoMirror
      resourceName=None
      resourceObject=null
      tiling=NoTile
      imageSize=<0 0>
      margin=<ltrb 0 0 0 0>
      outlineSettings {
       bUseBrushTransparency=false
       roundingType=HalfHeightRadius
       width=0
       color {colorUseRule=UseColor_Specified, specifiedColor=<rgba 0 0 0 0>}
       cornerRadii=<0 0 0 0>
      }
      tintColor {colorUseRule=UseColor_Specified, specifiedColor=<rgba 0.0252 0.0319 0.0423 1>}
      uVRegion {bIsValid=false, max=<0 0>, min=<0 0>}
     }
     inactiveHoveredBrush {
      bIsDynamicallyLoaded=false
      drawAs=Image
      imageType=NoImage
      mirroring=NoMirror
      resourceName=None
      resourceObject=null
      tiling=NoTile
      imageSize=<0 0>
      margin=<ltrb 0 0 0 0>
      outlineSettings {
       bUseBrushTransparency=false
       roundingType=HalfHeightRadius
       width=0
       color {colorUseRule=UseColor_Specified, specifiedColor=<rgba 0 0 0 0>}
       cornerRadii=<0 0 0 0>
      }
      tintColor {colorUseRule=UseColor_Specified, specifiedColor=<rgba 0.0513 0.0953 0.159 1>}
      uVRegion {bIsValid=false, max=<0 0>, min=<0 0>}
     }
     oddRowBackgroundBrush {
      bIsDynamicallyLoaded=false
      drawAs=Image
      imageType=NoImage
      mirroring=NoMirror
      resourceName=None
      resourceObject=null
      tiling=NoTile
      imageSize=<0 0>
      margin=<ltrb 0 0 0 0>
      outlineSettings {
       bUseBrushTransparency=false
       roundingType=HalfHeightRadius
       width=0
       color {colorUseRule=UseColor_Specified, specifiedColor=<rgba 0 0 0 0>}
       cornerRadii=<0 0 0 0>
      }
      tintColor {colorUseRule=UseColor_Specified, specifiedColor=<rgba 0.0103 0.0103 0.0103 1>}
      uVRegion {bIsValid=false, max=<0 0>, min=<0 0>}
     }
     oddRowBackgroundHoveredBrush {
      bIsDynamicallyLoaded=false
      drawAs=Image
      imageType=NoImage
      mirroring=NoMirror
      resourceName=None
      resourceObject=null
      tiling=NoTile
      imageSize=<0 0>
      margin=<ltrb 0 0 0 0>
      outlineSettings {
       bUseBrushTransparency=false
       roundingType=HalfHeightRadius
       width=0
       color {colorUseRule=UseColor_Specified, specifiedColor=<rgba 0 0 0 0>}
       cornerRadii=<0 0 0 0>
      }
      tintColor {colorUseRule=UseColor_Specified, specifiedColor=<rgba 0.0176 0.0176 0.0176 1>}
      uVRegion {bIsValid=false, max=<0 0>, min=<0 0>}
     }
     parentRowBackgroundBrush {
      bIsDynamicallyLoaded=false
      drawAs=Image
      imageType=NoImage
      mirroring=NoMirror
      resourceName=None
      resourceObject=null
      tiling=NoTile
      imageSize=<32 32>
      margin=<ltrb 0 0 0 0>
      outlineSettings {
       bUseBrushTransparency=false
       roundingType=HalfHeightRadius
       width=0
       color {colorUseRule=UseColor_Specified, specifiedColor=<rgba 0 0 0 0>}
       cornerRadii=<0 0 0 0>
      }
      tintColor {colorUseRule=UseColor_Specified, specifiedColor=<rgba 1 1 1 1>}
      uVRegion {bIsValid=false, max=<0 0>, min=<0 0>}
     }
     parentRowBackgroundHoveredBrush {
      bIsDynamicallyLoaded=false
      drawAs=Image
      imageType=NoImage
      mirroring=NoMirror
      resourceName=None
      resourceObject=null
      tiling=NoTile
      imageSize=<32 32>
      margin=<ltrb 0 0 0 0>
      outlineSettings {
       bUseBrushTransparency=false
       roundingType=HalfHeightRadius
       width=0
       color {colorUseRule=UseColor_Specified, specifiedColor=<rgba 0 0 0 0>}
       cornerRadii=<0 0 0 0>
      }
      tintColor {colorUseRule=UseColor_Specified, specifiedColor=<rgba 1 1 1 1>}
      uVRegion {bIsValid=false, max=<0 0>, min=<0 0>}
     }
     selectedTextColor {colorUseRule=UseColor_Specified, specifiedColor=<rgba 1 1 1 1>}
     selectorFocusedBrush {
      bIsDynamicallyLoaded=false
      drawAs=Border
      imageType=FullColor
      mirroring=NoMirror
      resourceName=../../../Engine/Content/Slate/Common/Selector.png
      resourceObject=null
      tiling=Both
      imageSize=<0 0>
      margin=<ltrb 0.25 0.25 0.25 0.25>
      outlineSettings {
       bUseBrushTransparency=false
       roundingType=HalfHeightRadius
       width=0
       color {colorUseRule=UseColor_Specified, specifiedColor=<rgba 0 0 0 0>}
       cornerRadii=<0 0 0 0>
      }
      tintColor {colorUseRule=UseColor_Specified, specifiedColor=<rgba 0.701 0.225 0.003 1>}
      uVRegion {bIsValid=false, max=<0 0>, min=<0 0>}
     }
     textColor {colorUseRule=UseColor_Specified, specifiedColor=<rgba 0.5271 0.5271 0.5271 1>}
    }
    widgetStyle {
     comboButtonStyle {
      downArrowAlign=VAlign_Center
      buttonStyle {
       clickedSlateSound {resourceObject=null}
       disabled {
        bIsDynamicallyLoaded=false
        drawAs=RoundedBox
        imageType=NoImage
        mirroring=NoMirror
        resourceName=None
        resourceObject=null
        tiling=NoTile
        imageSize=<0 0>
        margin=<ltrb 0 0 0 0>
        outlineSettings {
         bUseBrushTransparency=false
         roundingType=FixedRadius
         width=1
         color {colorUseRule=UseColor_ColorTable, specifiedColor=<rgba 0 0 0 0>}
         cornerRadii=<4 4 4 4>
        }
        tintColor {colorUseRule=UseColor_Specified, specifiedColor=<rgba 0.0395 0.0395 0.0395 1>}
        uVRegion {bIsValid=false, max=<0 0>, min=<0 0>}
       }
       disabledForeground {colorUseRule=UseColor_ColorTable, specifiedColor=<rgba 0 0 0 0>}
       hovered {
        bIsDynamicallyLoaded=false
        drawAs=RoundedBox
        imageType=NoImage
        mirroring=NoMirror
        resourceName=None
        resourceObject=null
        tiling=NoTile
        imageSize=<0 0>
        margin=<ltrb 0 0 0 0>
        outlineSettings {
         bUseBrushTransparency=false
         roundingType=FixedRadius
         width=1
         color {colorUseRule=UseColor_ColorTable, specifiedColor=<rgba 0 0 0 0>}
         cornerRadii=<4 4 4 4>
        }
        tintColor {colorUseRule=UseColor_Specified, specifiedColor=<rgba 0.0048 0.0048 0.0048 1>}
        uVRegion {bIsValid=false, max=<0 0>, min=<0 0>}
       }
       hoveredForeground {colorUseRule=UseColor_ColorTable, specifiedColor=<rgba 0 0 0 0>}
       hoveredSlateSound {resourceObject=null}
       normal {
        bIsDynamicallyLoaded=false
        drawAs=RoundedBox
        imageType=NoImage
        mirroring=NoMirror
        resourceName=None
        resourceObject=null
        tiling=NoTile
        imageSize=<0 0>
        margin=<ltrb 0 0 0 0>
        outlineSettings {
         bUseBrushTransparency=false
         roundingType=FixedRadius
         width=1
         color {colorUseRule=UseColor_ColorTable, specifiedColor=<rgba 0 0 0 0>}
         cornerRadii=<4 4 4 4>
        }
        tintColor {colorUseRule=UseColor_Specified, specifiedColor=<rgba 0.0048 0.0048 0.0048 1>}
        uVRegion {bIsValid=false, max=<0 0>, min=<0 0>}
       }
       normalForeground {colorUseRule=UseColor_ColorTable, specifiedColor=<rgba 0 0 0 0>}
       normalPadding=<ltrb 8 1 8 1>
       pressed {
        bIsDynamicallyLoaded=false
        drawAs=RoundedBox
        imageType=NoImage
        mirroring=NoMirror
        resourceName=None
        resourceObject=null
        tiling=NoTile
        imageSize=<0 0>
        margin=<ltrb 0 0 0 0>
        outlineSettings {
         bUseBrushTransparency=false
         roundingType=FixedRadius
         width=1
         color {colorUseRule=UseColor_ColorTable, specifiedColor=<rgba 0 0 0 0>}
         cornerRadii=<4 4 4 4>
        }
        tintColor {colorUseRule=UseColor_Specified, specifiedColor=<rgba 0.0048 0.0048 0.0048 1>}
        uVRegion {bIsValid=false, max=<0 0>, min=<0 0>}
       }
       pressedForeground {colorUseRule=UseColor_ColorTable, specifiedColor=<rgba 0 0 0 0>}
       pressedPadding=<ltrb 8 2 8 0>
       pressedSlateSound {resourceObject=null}
      }
      contentPadding=<ltrb 0 0 0 0>
      downArrowImage {
       bIsDynamicallyLoaded=false
       drawAs=Image
       imageType=Vector
       mirroring=NoMirror
       resourceName=../../../Engine/Content/Slate/Starship/CoreWidgets/ComboBox/wide-chevron-down.svg
       resourceObject=null
       tiling=NoTile
       imageSize=<16 16>
       margin=<ltrb 0 0 0 0>
       outlineSettings {
        bUseBrushTransparency=false
        roundingType=HalfHeightRadius
        width=0
        color {colorUseRule=UseColor_Specified, specifiedColor=<rgba 0 0 0 0>}
        cornerRadii=<0 0 0 0>
       }
       tintColor {colorUseRule=UseColor_Specified, specifiedColor=<rgba 1 1 1 1>}
       uVRegion {bIsValid=false, max=<0 0>, min=<0 0>}
      }
      downArrowPadding=<ltrb 2 2 2 2>
      menuBorderBrush {
       bIsDynamicallyLoaded=false
       drawAs=RoundedBox
       imageType=NoImage
       mirroring=NoMirror
       resourceName=None
       resourceObject=null
       tiling=NoTile
       imageSize=<0 0>
       margin=<ltrb 0 0 0 0>
       outlineSettings {
        bUseBrushTransparency=false
        roundingType=FixedRadius
        width=1
        color {colorUseRule=UseColor_ColorTable, specifiedColor=<rgba 0 0 0 0>}
        cornerRadii=<0 0 0 0>
       }
       tintColor {colorUseRule=UseColor_Specified, specifiedColor=<rgba 0.0395 0.0395 0.0395 1>}
       uVRegion {bIsValid=false, max=<0 0>, min=<0 0>}
      }
      menuBorderPadding=<ltrb 0 0 0 0>
      shadowColorAndOpacity=<rgba 0 0 0 1>
      shadowOffset=<0 0>
     }
     contentPadding=<ltrb 0 0 0 0>
     menuRowPadding=<ltrb 8 3 8 3>
     pressedSlateSound {resourceObject=null}
     selectionChangeSlateSound {resourceObject=null}
    }
   }
   slot {class=VerticalBoxSlot}
  }
  VariationHorizontalBox {
   widgetClassPath=&HorizontalBox
   bIsVariable=true
   parent=NameVerticalBox
   widget=VariationHorizontalBox
   slot {class=VerticalBoxSlot, properties {padding=<ltrb 0 2 0 1>}}
  }
  SizeBox_7 {
   widgetClassPath=&SizeBox
   parent=VariationHorizontalBox
   widget=SizeBox_7
   properties {bOverride_MinDesiredWidth=true, minDesiredWidth=100}
   slot {class=HorizontalBoxSlot, properties {verticalAlignment=VAlign_Center}}
  }
  VariationTextBlock {
   widgetClassPath=&TextBlock
   bIsVariable=true
   parent=SizeBox_7
   widget=VariationTextBlock
   properties {
    text=Mood/Status
    toolTipText=What Mood/State am I in?
    colorAndOpacity {colorUseRule=UseColor_Specified, specifiedColor=<rgba 0 0.5255 0.3373 1>}
    font {
     bForceMonospaced=false
     bMaterialIsStencil=false
     fontMaterial=null
     fontObject=&Roboto
     letterSpacing=0
     monospacedWidth=1
     size=10
     skewAmount=0
     typefaceFontName=Bold
     outlineSettings {
      bApplyOutlineToDropShadows=false
      bMiteredCorners=false
      bSeparateFillAlpha=false
      outlineMaterial=null
      outlineSize=0
      outlineColor=<rgba 0 0 0 1>
     }
    }
   }
   slot {class=SizeBoxSlot, properties {horizontalAlignment=HAlign_Right}}
  }
  VerticalBox_6 {
   widgetClassPath=&VerticalBox
   parent=VariationHorizontalBox
   widget=VerticalBox_6
   slot {class=HorizontalBoxSlot, properties {padding=<ltrb 5 0 0 0>, size {sizeRule=Fill, value=1}}}
  }
  VariationComboBox {
   widgetClassPath=&SearchableStringComboBox
   bIsVariable=true
   parent=VerticalBox_6
   widget=VariationComboBox
   properties {
    font {
     bForceMonospaced=false
     bMaterialIsStencil=false
     fontMaterial=null
     fontObject=&Roboto
     letterSpacing=0
     monospacedWidth=1
     size=10
     skewAmount=0
     typefaceFontName=Regular
     outlineSettings {
      bApplyOutlineToDropShadows=false
      bMiteredCorners=false
      bSeparateFillAlpha=false
      outlineMaterial=null
      outlineSize=0
      outlineColor=<rgba 0 0 0 1>
     }
    }
    foregroundColor {colorUseRule=UseColor_Specified, specifiedColor=<rgba 0.5271 0.5271 0.5271 1>}
    itemStyle {
     bUseParentRowBrush=false
     activeBrush {
      bIsDynamicallyLoaded=false
      drawAs=Image
      imageType=NoImage
      mirroring=NoMirror
      resourceName=None
      resourceObject=null
      tiling=NoTile
      imageSize=<0 0>
      margin=<ltrb 0 0 0 0>
      outlineSettings {
       bUseBrushTransparency=false
       roundingType=HalfHeightRadius
       width=0
       color {colorUseRule=UseColor_Specified, specifiedColor=<rgba 0 0 0 0>}
       cornerRadii=<0 0 0 0>
      }
      tintColor {colorUseRule=UseColor_Specified, specifiedColor=<rgba 0 0.162 0.7454 1>}
      uVRegion {bIsValid=false, max=<0 0>, min=<0 0>}
     }
     activeHighlightedBrush {
      bIsDynamicallyLoaded=false
      drawAs=Image
      imageType=NoImage
      mirroring=NoMirror
      resourceName=None
      resourceObject=null
      tiling=NoTile
      imageSize=<0 0>
      margin=<ltrb 0 0 0 0>
      outlineSettings {
       bUseBrushTransparency=false
       roundingType=HalfHeightRadius
       width=0
       color {colorUseRule=UseColor_Specified, specifiedColor=<rgba 0 0 0 0>}
       cornerRadii=<0 0 0 0>
      }
      tintColor {colorUseRule=UseColor_Specified, specifiedColor=<rgba 0.0252 0.0319 0.0423 1>}
      uVRegion {bIsValid=false, max=<0 0>, min=<0 0>}
     }
     activeHoveredBrush {
      bIsDynamicallyLoaded=false
      drawAs=Image
      imageType=NoImage
      mirroring=NoMirror
      resourceName=None
      resourceObject=null
      tiling=NoTile
      imageSize=<0 0>
      margin=<ltrb 0 0 0 0>
      outlineSettings {
       bUseBrushTransparency=false
       roundingType=HalfHeightRadius
       width=0
       color {colorUseRule=UseColor_Specified, specifiedColor=<rgba 0 0 0 0>}
       cornerRadii=<0 0 0 0>
      }
      tintColor {colorUseRule=UseColor_Specified, specifiedColor=<rgba 0 0.162 0.7454 1>}
      uVRegion {bIsValid=false, max=<0 0>, min=<0 0>}
     }
     dropIndicator_Above {
      bIsDynamicallyLoaded=false
      drawAs=Box
      imageType=FullColor
      mirroring=NoMirror
      resourceName=../../../Engine/Content/Slate/Common/DropZoneIndicator_Above.png
      resourceObject=null
      tiling=NoTile
      imageSize=<0 0>
      margin=<ltrb 0.625 0.625 0 0>
      outlineSettings {
       bUseBrushTransparency=false
       roundingType=HalfHeightRadius
       width=0
       color {colorUseRule=UseColor_Specified, specifiedColor=<rgba 0 0 0 0>}
       cornerRadii=<0 0 0 0>
      }
      tintColor {colorUseRule=UseColor_Specified, specifiedColor=<rgba 0.0091 0.3515 0.9647 1>}
      uVRegion {bIsValid=false, max=<0 0>, min=<0 0>}
     }
     dropIndicator_Below {
      bIsDynamicallyLoaded=false
      drawAs=Box
      imageType=FullColor
      mirroring=NoMirror
      resourceName=../../../Engine/Content/Slate/Common/DropZoneIndicator_Below.png
      resourceObject=null
      tiling=NoTile
      imageSize=<0 0>
      margin=<ltrb 0.625 0 0 0.625>
      outlineSettings {
       bUseBrushTransparency=false
       roundingType=HalfHeightRadius
       width=0
       color {colorUseRule=UseColor_Specified, specifiedColor=<rgba 0 0 0 0>}
       cornerRadii=<0 0 0 0>
      }
      tintColor {colorUseRule=UseColor_Specified, specifiedColor=<rgba 0.0091 0.3515 0.9647 1>}
      uVRegion {bIsValid=false, max=<0 0>, min=<0 0>}
     }
     dropIndicator_Onto {
      bIsDynamicallyLoaded=false
      drawAs=Box
      imageType=FullColor
      mirroring=NoMirror
      resourceName=../../../Engine/Content/Slate/Common/DropZoneIndicator_Onto.png
      resourceObject=null
      tiling=NoTile
      imageSize=<0 0>
      margin=<ltrb 0.25 0.25 0.25 0.25>
      outlineSettings {
       bUseBrushTransparency=false
       roundingType=HalfHeightRadius
       width=0
       color {colorUseRule=UseColor_Specified, specifiedColor=<rgba 0 0 0 0>}
       cornerRadii=<0 0 0 0>
      }
      tintColor {colorUseRule=UseColor_Specified, specifiedColor=<rgba 0.0091 0.3515 0.9647 1>}
      uVRegion {bIsValid=false, max=<0 0>, min=<0 0>}
     }
     evenRowBackgroundBrush {
      bIsDynamicallyLoaded=false
      drawAs=Image
      imageType=NoImage
      mirroring=NoMirror
      resourceName=None
      resourceObject=null
      tiling=NoTile
      imageSize=<0 0>
      margin=<ltrb 0 0 0 0>
      outlineSettings {
       bUseBrushTransparency=false
       roundingType=HalfHeightRadius
       width=0
       color {colorUseRule=UseColor_Specified, specifiedColor=<rgba 0 0 0 0>}
       cornerRadii=<0 0 0 0>
      }
      tintColor {colorUseRule=UseColor_Specified, specifiedColor=<rgba 0.0103 0.0103 0.0103 1>}
      uVRegion {bIsValid=false, max=<0 0>, min=<0 0>}
     }
     evenRowBackgroundHoveredBrush {
      bIsDynamicallyLoaded=false
      drawAs=Image
      imageType=NoImage
      mirroring=NoMirror
      resourceName=None
      resourceObject=null
      tiling=NoTile
      imageSize=<0 0>
      margin=<ltrb 0 0 0 0>
      outlineSettings {
       bUseBrushTransparency=false
       roundingType=HalfHeightRadius
       width=0
       color {colorUseRule=UseColor_Specified, specifiedColor=<rgba 0 0 0 0>}
       cornerRadii=<0 0 0 0>
      }
      tintColor {colorUseRule=UseColor_Specified, specifiedColor=<rgba 0.0176 0.0176 0.0176 1>}
      uVRegion {bIsValid=false, max=<0 0>, min=<0 0>}
     }
     inactiveBrush {
      bIsDynamicallyLoaded=false
      drawAs=Image
      imageType=NoImage
      mirroring=NoMirror
      resourceName=None
      resourceObject=null
      tiling=NoTile
      imageSize=<0 0>
      margin=<ltrb 0 0 0 0>
      outlineSettings {
       bUseBrushTransparency=false
       roundingType=HalfHeightRadius
       width=0
       color {colorUseRule=UseColor_Specified, specifiedColor=<rgba 0 0 0 0>}
       cornerRadii=<0 0 0 0>
      }
      tintColor {colorUseRule=UseColor_Specified, specifiedColor=<rgba 0.0513 0.0953 0.159 1>}
      uVRegion {bIsValid=false, max=<0 0>, min=<0 0>}
     }
     inactiveHighlightedBrush {
      bIsDynamicallyLoaded=false
      drawAs=Image
      imageType=NoImage
      mirroring=NoMirror
      resourceName=None
      resourceObject=null
      tiling=NoTile
      imageSize=<0 0>
      margin=<ltrb 0 0 0 0>
      outlineSettings {
       bUseBrushTransparency=false
       roundingType=HalfHeightRadius
       width=0
       color {colorUseRule=UseColor_Specified, specifiedColor=<rgba 0 0 0 0>}
       cornerRadii=<0 0 0 0>
      }
      tintColor {colorUseRule=UseColor_Specified, specifiedColor=<rgba 0.0252 0.0319 0.0423 1>}
      uVRegion {bIsValid=false, max=<0 0>, min=<0 0>}
     }
     inactiveHoveredBrush {
      bIsDynamicallyLoaded=false
      drawAs=Image
      imageType=NoImage
      mirroring=NoMirror
      resourceName=None
      resourceObject=null
      tiling=NoTile
      imageSize=<0 0>
      margin=<ltrb 0 0 0 0>
      outlineSettings {
       bUseBrushTransparency=false
       roundingType=HalfHeightRadius
       width=0
       color {colorUseRule=UseColor_Specified, specifiedColor=<rgba 0 0 0 0>}
       cornerRadii=<0 0 0 0>
      }
      tintColor {colorUseRule=UseColor_Specified, specifiedColor=<rgba 0.0513 0.0953 0.159 1>}
      uVRegion {bIsValid=false, max=<0 0>, min=<0 0>}
     }
     oddRowBackgroundBrush {
      bIsDynamicallyLoaded=false
      drawAs=Image
      imageType=NoImage
      mirroring=NoMirror
      resourceName=None
      resourceObject=null
      tiling=NoTile
      imageSize=<0 0>
      margin=<ltrb 0 0 0 0>
      outlineSettings {
       bUseBrushTransparency=false
       roundingType=HalfHeightRadius
       width=0
       color {colorUseRule=UseColor_Specified, specifiedColor=<rgba 0 0 0 0>}
       cornerRadii=<0 0 0 0>
      }
      tintColor {colorUseRule=UseColor_Specified, specifiedColor=<rgba 0.0103 0.0103 0.0103 1>}
      uVRegion {bIsValid=false, max=<0 0>, min=<0 0>}
     }
     oddRowBackgroundHoveredBrush {
      bIsDynamicallyLoaded=false
      drawAs=Image
      imageType=NoImage
      mirroring=NoMirror
      resourceName=None
      resourceObject=null
      tiling=NoTile
      imageSize=<0 0>
      margin=<ltrb 0 0 0 0>
      outlineSettings {
       bUseBrushTransparency=false
       roundingType=HalfHeightRadius
       width=0
       color {colorUseRule=UseColor_Specified, specifiedColor=<rgba 0 0 0 0>}
       cornerRadii=<0 0 0 0>
      }
      tintColor {colorUseRule=UseColor_Specified, specifiedColor=<rgba 0.0176 0.0176 0.0176 1>}
      uVRegion {bIsValid=false, max=<0 0>, min=<0 0>}
     }
     parentRowBackgroundBrush {
      bIsDynamicallyLoaded=false
      drawAs=Image
      imageType=NoImage
      mirroring=NoMirror
      resourceName=None
      resourceObject=null
      tiling=NoTile
      imageSize=<32 32>
      margin=<ltrb 0 0 0 0>
      outlineSettings {
       bUseBrushTransparency=false
       roundingType=HalfHeightRadius
       width=0
       color {colorUseRule=UseColor_Specified, specifiedColor=<rgba 0 0 0 0>}
       cornerRadii=<0 0 0 0>
      }
      tintColor {colorUseRule=UseColor_Specified, specifiedColor=<rgba 1 1 1 1>}
      uVRegion {bIsValid=false, max=<0 0>, min=<0 0>}
     }
     parentRowBackgroundHoveredBrush {
      bIsDynamicallyLoaded=false
      drawAs=Image
      imageType=NoImage
      mirroring=NoMirror
      resourceName=None
      resourceObject=null
      tiling=NoTile
      imageSize=<32 32>
      margin=<ltrb 0 0 0 0>
      outlineSettings {
       bUseBrushTransparency=false
       roundingType=HalfHeightRadius
       width=0
       color {colorUseRule=UseColor_Specified, specifiedColor=<rgba 0 0 0 0>}
       cornerRadii=<0 0 0 0>
      }
      tintColor {colorUseRule=UseColor_Specified, specifiedColor=<rgba 1 1 1 1>}
      uVRegion {bIsValid=false, max=<0 0>, min=<0 0>}
     }
     selectedTextColor {colorUseRule=UseColor_Specified, specifiedColor=<rgba 1 1 1 1>}
     selectorFocusedBrush {
      bIsDynamicallyLoaded=false
      drawAs=Border
      imageType=FullColor
      mirroring=NoMirror
      resourceName=../../../Engine/Content/Slate/Common/Selector.png
      resourceObject=null
      tiling=Both
      imageSize=<0 0>
      margin=<ltrb 0.25 0.25 0.25 0.25>
      outlineSettings {
       bUseBrushTransparency=false
       roundingType=HalfHeightRadius
       width=0
       color {colorUseRule=UseColor_Specified, specifiedColor=<rgba 0 0 0 0>}
       cornerRadii=<0 0 0 0>
      }
      tintColor {colorUseRule=UseColor_Specified, specifiedColor=<rgba 0.701 0.225 0.003 1>}
      uVRegion {bIsValid=false, max=<0 0>, min=<0 0>}
     }
     textColor {colorUseRule=UseColor_Specified, specifiedColor=<rgba 0.5271 0.5271 0.5271 1>}
    }
    widgetStyle {
     comboButtonStyle {
      downArrowAlign=VAlign_Center
      buttonStyle {
       clickedSlateSound {resourceObject=null}
       disabled {
        bIsDynamicallyLoaded=false
        drawAs=RoundedBox
        imageType=NoImage
        mirroring=NoMirror
        resourceName=None
        resourceObject=null
        tiling=NoTile
        imageSize=<0 0>
        margin=<ltrb 0 0 0 0>
        outlineSettings {
         bUseBrushTransparency=false
         roundingType=FixedRadius
         width=1
         color {colorUseRule=UseColor_ColorTable, specifiedColor=<rgba 0 0 0 0>}
         cornerRadii=<4 4 4 4>
        }
        tintColor {colorUseRule=UseColor_Specified, specifiedColor=<rgba 0.0395 0.0395 0.0395 1>}
        uVRegion {bIsValid=false, max=<0 0>, min=<0 0>}
       }
       disabledForeground {colorUseRule=UseColor_ColorTable, specifiedColor=<rgba 0 0 0 0>}
       hovered {
        bIsDynamicallyLoaded=false
        drawAs=RoundedBox
        imageType=NoImage
        mirroring=NoMirror
        resourceName=None
        resourceObject=null
        tiling=NoTile
        imageSize=<0 0>
        margin=<ltrb 0 0 0 0>
        outlineSettings {
         bUseBrushTransparency=false
         roundingType=FixedRadius
         width=1
         color {colorUseRule=UseColor_ColorTable, specifiedColor=<rgba 0 0 0 0>}
         cornerRadii=<4 4 4 4>
        }
        tintColor {colorUseRule=UseColor_Specified, specifiedColor=<rgba 0.0048 0.0048 0.0048 1>}
        uVRegion {bIsValid=false, max=<0 0>, min=<0 0>}
       }
       hoveredForeground {colorUseRule=UseColor_ColorTable, specifiedColor=<rgba 0 0 0 0>}
       hoveredSlateSound {resourceObject=null}
       normal {
        bIsDynamicallyLoaded=false
        drawAs=RoundedBox
        imageType=NoImage
        mirroring=NoMirror
        resourceName=None
        resourceObject=null
        tiling=NoTile
        imageSize=<0 0>
        margin=<ltrb 0 0 0 0>
        outlineSettings {
         bUseBrushTransparency=false
         roundingType=FixedRadius
         width=1
         color {colorUseRule=UseColor_ColorTable, specifiedColor=<rgba 0 0 0 0>}
         cornerRadii=<4 4 4 4>
        }
        tintColor {colorUseRule=UseColor_Specified, specifiedColor=<rgba 0.0048 0.0048 0.0048 1>}
        uVRegion {bIsValid=false, max=<0 0>, min=<0 0>}
       }
       normalForeground {colorUseRule=UseColor_ColorTable, specifiedColor=<rgba 0 0 0 0>}
       normalPadding=<ltrb 8 1 8 1>
       pressed {
        bIsDynamicallyLoaded=false
        drawAs=RoundedBox
        imageType=NoImage
        mirroring=NoMirror
        resourceName=None
        resourceObject=null
        tiling=NoTile
        imageSize=<0 0>
        margin=<ltrb 0 0 0 0>
        outlineSettings {
         bUseBrushTransparency=false
         roundingType=FixedRadius
         width=1
         color {colorUseRule=UseColor_ColorTable, specifiedColor=<rgba 0 0 0 0>}
         cornerRadii=<4 4 4 4>
        }
        tintColor {colorUseRule=UseColor_Specified, specifiedColor=<rgba 0.0048 0.0048 0.0048 1>}
        uVRegion {bIsValid=false, max=<0 0>, min=<0 0>}
       }
       pressedForeground {colorUseRule=UseColor_ColorTable, specifiedColor=<rgba 0 0 0 0>}
       pressedPadding=<ltrb 8 2 8 0>
       pressedSlateSound {resourceObject=null}
      }
      contentPadding=<ltrb 0 0 0 0>
      downArrowImage {
       bIsDynamicallyLoaded=false
       drawAs=Image
       imageType=Vector
       mirroring=NoMirror
       resourceName=../../../Engine/Content/Slate/Starship/CoreWidgets/ComboBox/wide-chevron-down.svg
       resourceObject=null
       tiling=NoTile
       imageSize=<16 16>
       margin=<ltrb 0 0 0 0>
       outlineSettings {
        bUseBrushTransparency=false
        roundingType=HalfHeightRadius
        width=0
        color {colorUseRule=UseColor_Specified, specifiedColor=<rgba 0 0 0 0>}
        cornerRadii=<0 0 0 0>
       }
       tintColor {colorUseRule=UseColor_Specified, specifiedColor=<rgba 1 1 1 1>}
       uVRegion {bIsValid=false, max=<0 0>, min=<0 0>}
      }
      downArrowPadding=<ltrb 2 2 2 2>
      menuBorderBrush {
       bIsDynamicallyLoaded=false
       drawAs=RoundedBox
       imageType=NoImage
       mirroring=NoMirror
       resourceName=None
       resourceObject=null
       tiling=NoTile
       imageSize=<0 0>
       margin=<ltrb 0 0 0 0>
       outlineSettings {
        bUseBrushTransparency=false
        roundingType=FixedRadius
        width=1
        color {colorUseRule=UseColor_ColorTable, specifiedColor=<rgba 0 0 0 0>}
        cornerRadii=<0 0 0 0>
       }
       tintColor {colorUseRule=UseColor_Specified, specifiedColor=<rgba 0.0395 0.0395 0.0395 1>}
       uVRegion {bIsValid=false, max=<0 0>, min=<0 0>}
      }
      menuBorderPadding=<ltrb 0 0 0 0>
      shadowColorAndOpacity=<rgba 0 0 0 1>
      shadowOffset=<0 0>
     }
     contentPadding=<ltrb 0 0 0 0>
     menuRowPadding=<ltrb 8 3 8 3>
     pressedSlateSound {resourceObject=null}
     selectionChangeSlateSound {resourceObject=null}
    }
   }
   slot {class=VerticalBoxSlot}
  }
  OptionalTextBlock {
   widgetClassPath=&TextBlock
   bIsVariable=true
   parent=NameVerticalBox
   widget=OptionalTextBlock
   properties {
    text=* Optional
    colorAndOpacity {colorUseRule=UseColor_Specified, specifiedColor=<rgba 0 0.5255 0.3373 1>}
    font {
     bForceMonospaced=false
     bMaterialIsStencil=false
     fontMaterial=null
     fontObject=&Roboto
     letterSpacing=0
     monospacedWidth=1
     size=7.5
     skewAmount=0
     typefaceFontName=Bold
     outlineSettings {
      bApplyOutlineToDropShadows=false
      bMiteredCorners=false
      bSeparateFillAlpha=false
      outlineMaterial=null
      outlineSize=0
      outlineColor=<rgba 0 0 0 1>
     }
    }
   }
   slot {class=VerticalBoxSlot, properties {horizontalAlignment=HAlign_Right, size {sizeRule=Fill, value=1}}}
  }
  FeedbackWidget {
   widgetClassPath=&EUW_AnimationDialog_BorderInfo
   bIsVariable=true
   parent=EditorUtilityScrollBox_0
   widget=FeedbackWidget
   properties {visibility=Collapsed}
   slot {class=ScrollBoxSlot, properties {padding=<ltrb 0 2.5 0 2.5>}}
  }
  ButtonBox {widgetClassPath=&HorizontalBox, parent=EditorUtilityScrollBox_0, widget=ButtonBox, slot {class=ScrollBoxSlot}}
  CreateButton {
   widgetClassPath=&Button
   parent=ButtonBox
   widget=CreateButton
   properties {widgetStyle {
     clickedSlateSound {resourceObject=null}
     disabled {
      bIsDynamicallyLoaded=false
      drawAs=RoundedBox
      imageType=NoImage
      mirroring=NoMirror
      resourceName=None
      resourceObject=null
      tiling=NoTile
      imageSize=<0 0>
      margin=<ltrb 0 0 0 0>
      outlineSettings {
       bUseBrushTransparency=false
       roundingType=FixedRadius
       width=1
       color {colorUseRule=UseColor_ColorTable, specifiedColor=<rgba 0 0 0 0>}
       cornerRadii=<4 4 4 4>
      }
      tintColor {colorUseRule=UseColor_Specified, specifiedColor=<rgba 0.0395 0.0395 0.0395 1>}
      uVRegion {bIsValid=false, max=<0 0>, min=<0 0>}
     }
     disabledForeground {colorUseRule=UseColor_ColorTable, specifiedColor=<rgba 0 0 0 0>}
     hovered {
      bIsDynamicallyLoaded=false
      drawAs=RoundedBox
      imageType=NoImage
      mirroring=NoMirror
      resourceName=None
      resourceObject=null
      tiling=NoTile
      imageSize=<32 32>
      margin=<ltrb 0 0 0 0>
      outlineSettings {
       bUseBrushTransparency=false
       roundingType=FixedRadius
       width=0
       color {colorUseRule=UseColor_ColorTable, specifiedColor=<rgba 0 0 0 0>}
       cornerRadii=<4 4 4 4>
      }
      tintColor {colorUseRule=UseColor_Specified, specifiedColor=<rgba 0.0953 0.0953 0.0953 1>}
      uVRegion {bIsValid=false, max=<0 0>, min=<0 0>}
     }
     hoveredForeground {colorUseRule=UseColor_ColorTable, specifiedColor=<rgba 0 0 0 0>}
     hoveredSlateSound {resourceObject=null}
     normal {
      bIsDynamicallyLoaded=false
      drawAs=RoundedBox
      imageType=NoImage
      mirroring=NoMirror
      resourceName=None
      resourceObject=null
      tiling=NoTile
      imageSize=<32 32>
      margin=<ltrb 0 0 0 0>
      outlineSettings {
       bUseBrushTransparency=false
       roundingType=FixedRadius
       width=0
       color {colorUseRule=UseColor_ColorTable, specifiedColor=<rgba 0 0 0 0>}
       cornerRadii=<4 4 4 4>
      }
      tintColor {colorUseRule=UseColor_Specified, specifiedColor=<rgba 0.0395 0.0395 0.0395 1>}
      uVRegion {bIsValid=false, max=<0 0>, min=<0 0>}
     }
     normalForeground {colorUseRule=UseColor_ColorTable, specifiedColor=<rgba 0 0 0 0>}
     normalPadding=<ltrb 12 1.5 12 1.5>
     pressed {
      bIsDynamicallyLoaded=false
      drawAs=RoundedBox
      imageType=NoImage
      mirroring=NoMirror
      resourceName=None
      resourceObject=null
      tiling=NoTile
      imageSize=<32 32>
      margin=<ltrb 0 0 0 0>
      outlineSettings {
       bUseBrushTransparency=false
       roundingType=FixedRadius
       width=0
       color {colorUseRule=UseColor_ColorTable, specifiedColor=<rgba 0 0 0 0>}
       cornerRadii=<4 4 4 4>
      }
      tintColor {colorUseRule=UseColor_Specified, specifiedColor=<rgba 0.0284 0.0284 0.0284 1>}
      uVRegion {bIsValid=false, max=<0 0>, min=<0 0>}
     }
     pressedForeground {colorUseRule=UseColor_ColorTable, specifiedColor=<rgba 0 0 0 0>}
     pressedPadding=<ltrb 12 2.5 12 0.5>
     pressedSlateSound {resourceObject=null}
    }}
   slot {class=HorizontalBoxSlot, properties {padding=<ltrb 0 0 2.5 0>, size {sizeRule=Fill, value=1}}}
  }
  CreateButtonTextBlock {
   widgetClassPath=&TextBlock
   bIsVariable=true
   parent=CreateButton
   widget=CreateButtonTextBlock
   properties {
    autoWrapText=true
    justification=Center
    text=Validate
    colorAndOpacity {colorUseRule=UseColor_Specified, specifiedColor=<rgba 0.5255 0.5255 0.5255 1>}
    font {
     bForceMonospaced=false
     bMaterialIsStencil=false
     fontMaterial=null
     fontObject=&Roboto
     letterSpacing=0
     monospacedWidth=1
     size=10
     skewAmount=0
     typefaceFontName=Bold
     outlineSettings {
      bApplyOutlineToDropShadows=false
      bMiteredCorners=false
      bSeparateFillAlpha=false
      outlineMaterial=null
      outlineSize=0
      outlineColor=<rgba 0 0 0 1>
     }
    }
   }
   slot {class=ButtonSlot}
  }
  CancelButton {
   widgetClassPath=&Button
   parent=ButtonBox
   widget=CancelButton
   properties {widgetStyle {
     clickedSlateSound {resourceObject=null}
     disabled {
      bIsDynamicallyLoaded=false
      drawAs=RoundedBox
      imageType=NoImage
      mirroring=NoMirror
      resourceName=None
      resourceObject=null
      tiling=NoTile
      imageSize=<0 0>
      margin=<ltrb 0 0 0 0>
      outlineSettings {
       bUseBrushTransparency=false
       roundingType=FixedRadius
       width=1
       color {colorUseRule=UseColor_ColorTable, specifiedColor=<rgba 0 0 0 0>}
       cornerRadii=<4 4 4 4>
      }
      tintColor {colorUseRule=UseColor_Specified, specifiedColor=<rgba 0.0395 0.0395 0.0395 1>}
      uVRegion {bIsValid=false, max=<0 0>, min=<0 0>}
     }
     disabledForeground {colorUseRule=UseColor_ColorTable, specifiedColor=<rgba 0 0 0 0>}
     hovered {
      bIsDynamicallyLoaded=false
      drawAs=RoundedBox
      imageType=NoImage
      mirroring=NoMirror
      resourceName=None
      resourceObject=null
      tiling=NoTile
      imageSize=<32 32>
      margin=<ltrb 0 0 0 0>
      outlineSettings {
       bUseBrushTransparency=false
       roundingType=FixedRadius
       width=0
       color {colorUseRule=UseColor_ColorTable, specifiedColor=<rgba 0 0 0 0>}
       cornerRadii=<4 4 4 4>
      }
      tintColor {colorUseRule=UseColor_Specified, specifiedColor=<rgba 0.0953 0.0953 0.0953 1>}
      uVRegion {bIsValid=false, max=<0 0>, min=<0 0>}
     }
     hoveredForeground {colorUseRule=UseColor_ColorTable, specifiedColor=<rgba 0 0 0 0>}
     hoveredSlateSound {resourceObject=null}
     normal {
      bIsDynamicallyLoaded=false
      drawAs=RoundedBox
      imageType=NoImage
      mirroring=NoMirror
      resourceName=None
      resourceObject=null
      tiling=NoTile
      imageSize=<32 32>
      margin=<ltrb 0 0 0 0>
      outlineSettings {
       bUseBrushTransparency=false
       roundingType=FixedRadius
       width=0
       color {colorUseRule=UseColor_ColorTable, specifiedColor=<rgba 0 0 0 0>}
       cornerRadii=<4 4 4 4>
      }
      tintColor {colorUseRule=UseColor_Specified, specifiedColor=<rgba 0.0395 0.0395 0.0395 1>}
      uVRegion {bIsValid=false, max=<0 0>, min=<0 0>}
     }
     normalForeground {colorUseRule=UseColor_ColorTable, specifiedColor=<rgba 0 0 0 0>}
     normalPadding=<ltrb 12 1.5 12 1.5>
     pressed {
      bIsDynamicallyLoaded=false
      drawAs=RoundedBox
      imageType=NoImage
      mirroring=NoMirror
      resourceName=None
      resourceObject=null
      tiling=NoTile
      imageSize=<32 32>
      margin=<ltrb 0 0 0 0>
      outlineSettings {
       bUseBrushTransparency=false
       roundingType=FixedRadius
       width=0
       color {colorUseRule=UseColor_ColorTable, specifiedColor=<rgba 0 0 0 0>}
       cornerRadii=<4 4 4 4>
      }
      tintColor {colorUseRule=UseColor_Specified, specifiedColor=<rgba 0.0284 0.0284 0.0284 1>}
      uVRegion {bIsValid=false, max=<0 0>, min=<0 0>}
     }
     pressedForeground {colorUseRule=UseColor_ColorTable, specifiedColor=<rgba 0 0 0 0>}
     pressedPadding=<ltrb 12 2.5 12 0.5>
     pressedSlateSound {resourceObject=null}
    }}
   slot {class=HorizontalBoxSlot, properties {padding=<ltrb 2.5 0 0 0>, size {sizeRule=Fill, value=1}}}
  }
  CancelTextBlock {
   widgetClassPath=&TextBlock
   parent=CancelButton
   widget=CancelTextBlock
   properties {
    autoWrapText=true
    justification=Center
    text=Cancel
    colorAndOpacity {colorUseRule=UseColor_Specified, specifiedColor=<rgba 0.5255 0.5255 0.5255 1>}
    font {
     bForceMonospaced=false
     bMaterialIsStencil=false
     fontMaterial=null
     fontObject=&Roboto
     letterSpacing=0
     monospacedWidth=1
     size=10
     skewAmount=0
     typefaceFontName=Bold
     outlineSettings {
      bApplyOutlineToDropShadows=false
      bMiteredCorners=false
      bSeparateFillAlpha=false
      outlineMaterial=null
      outlineSize=0
      outlineColor=<rgba 0 0 0 1>
     }
    }
   }
   slot {class=ButtonSlot}
  }
 ]
}
