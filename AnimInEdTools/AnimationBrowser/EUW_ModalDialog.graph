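# EUW_ModalDialog is an Editor Utility Widget that provides a reusable modal dialog frame for the AnimInEdTools animation browser.
# A translucent black root Border acts as a dimming backdrop around a centered, minimum-sized SizeBox.
# Inside it, a dark padded Border holds a variable NamedSlot where the dialog content is injected.
# It is layout only: the Event Graph holds just unconnected default events and there are no variables.
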WidgetBlueprint /AnimInEdTools/AnimationBrowser/EUW_ModalDialog
graphs [EventGraph {nodes [
   Event PreConstruct()
   Event Construct()
   Event Tick()
  ]}]
variables []
widgets {
 info {inheritedWidgetCount=0, parentClass=&EditorUtilityWidget, rootWidgetClass=&Border, widgetCount=4}
 widgets [
  Border_0 {
   widgetClassPath=&Border
   widget=Border_0
   properties {horizontalAlignment=HAlign_Center, verticalAlignment=VAlign_Center, brushColor=<rgba 0 0 0 0.7>}
  }
  SizeBox_0 {
   widgetClassPath=&SizeBox
   parent=Border_0
   widget=SizeBox_0
   properties {
    bOverride_MinDesiredHeight=true
    bOverride_MinDesiredWidth=true
    maxDesiredHeight=500
    maxDesiredWidth=500
    minDesiredHeight=200
    minDesiredWidth=300
   }
   slot {class=BorderSlot, properties {horizontalAlignment=HAlign_Center, verticalAlignment=VAlign_Center}}
  }
  Border_1 {
   widgetClassPath=&Border
   parent=SizeBox_0
   widget=Border_1
   properties {brushColor=<rgba 0.0157 0.0157 0.0157 1>, padding=<ltrb 5 5 5 5>}
   slot {class=SizeBoxSlot, properties {padding=<ltrb 5 5 5 5>}}
  }
  NamedSlot {
   widgetClassPath=&NamedSlot
   bIsVariable=true
   parent=Border_1
   widget=NamedSlot
   slot {class=BorderSlot, properties {padding=<ltrb 5 5 5 5>}}
  }
 ]
}
