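# EUW_RegistryLoadingDialog is an Editor Utility Widget in the AnimationBrowser tool that displays a notice that the Asset Registry is loading.
# A vertical box holds a single text block, DialogTextBlock, with centered gray bold text, vertically centered in a fill slot.
# The text block is exposed as a variable, but the event graph holds only unconnected default events, so the widget has no logic of its own.
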WidgetBlueprint /AnimInEdTools/AnimationBrowser/EUW_RegistryLoadingDialog
graphs [EventGraph {nodes [
   Event PreConstruct()
   Event Construct()
   Event Tick()
  ]}]
variables []
widgets {
 info {inheritedWidgetCount=0, parentClass=&EditorUtilityWidget, rootWidgetClass=&VerticalBox, widgetCount=2}
 widgets [
  VerticalBox_1 {widgetClassPath=&VerticalBox, widget=VerticalBox_1}
  DialogTextBlock {
   widgetClassPath=&TextBlock
   bIsVariable=true
   parent=VerticalBox_1
   widget=DialogTextBlock
   properties {
    justification=Center
    text=Asset Registry is loading...
    colorAndOpacity {colorUseRule=UseColor_Specified, specifiedColor=<rgba 0.5255 0.5255 0.5255 1>}
    font {
     bForceMonospaced=false
     bMaterialIsStencil=false
     fontMaterial=null
     fontObject=&Roboto
     letterSpacing=0
     monospacedWidth=1
     size=14
     skewAmount=0
     typefaceFontName=Bold
     outlineSettings {
      bApplyOutlineToDropShadows=false
      bMiteredCorners=false
      bSeparateFillAlpha=false
      outlineMaterial=null
      outlineSize=0
      outlineColor=<rgba 0 0 0 1>
     }
    }
   }
   slot {class=VerticalBoxSlot, properties {verticalAlignment=VAlign_Center, size {sizeRule=Fill, value=1}}}
  }
 ]
}
