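# Editor utility widget for an item row in the Animation Browser's Restore context menu.
# It pairs a label with an editable text field inside a rounded border.
# Set Autosave either turns the field into read-only "AutoSave" text with no border brush drawn, or hides field and border when false.
# It declares a File Path variable and a Context Menu Item Clicked dispatcher, neither used in its own graphs.
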WidgetBlueprint /AnimInEdTools/AnimationBrowser/Restore/EUW_Restore_ContextMenu_Item
graphs [
 Context Menu Item Clicked {nodes [Context Menu Item Clicked!() -> (then, Text:String) {is=title K2Node_FunctionEntry}]}
 EventGraph {nodes [
   Event PreConstruct()
   Event Construct()
   Event Tick()
  ]}
 Set Autosave {nodes [
   n1 = Set Autosave!() -> (then, Value:bool=~) {is=title K2Node_FunctionEntry}
   > n2 = Branch(Condition=@n1)
   > Set Text(Text="LOCTEXT(\"AutoSave\")", self:Editable Text Object Reference=Get EditableText() -> (EditableText:Editable Text Object Reference))
   > "SetIsReadOnly (Editable Text)"!(execute, self:Editable Text Object Reference=Get EditableText(), InbIsReadyOnly:bool=true) -> (then)
   n3 = Get Border() -> (Border:Border Object Reference)
   n4 = Get!(self:Border Object Reference=@n3) -> (
    Background_TintColor:Slate Color Structure
    Background_DrawAs:ESlateBrushDrawType Enum
    Background_Tiling:ESlateBrushTileType Enum
    Background_Mirroring:ESlateBrushMirrorType Enum
    Background_ImageSize:Vector 2D Structure
    Background_Margin:Margin Structure
    Background_ResourceObject:Object Reference
    Background_OutlineSettings:Slate Brush Outline Settings Structure
   )
   > SetBrush!(
    execute
    self:Border Object Reference=@n3
    InBrush_TintColor:Slate Color Structure=@n4
    InBrush_DrawAs:ESlateBrushDrawType Enum=NoDrawType
    InBrush_Tiling:ESlateBrushTileType Enum=@n4.Background_Tiling
    InBrush_Mirroring:ESlateBrushMirrorType Enum=@n4.Background_Mirroring
    InBrush_ImageSize:Vector 2D Structure=@n4.Background_ImageSize
    InBrush_Margin:Margin Structure=@n4.Background_Margin
    InBrush_ResourceObject:Object Reference=@n4.Background_ResourceObject
    InBrush_OutlineSettings:Slate Brush Outline Settings Structure=@n4.Background_OutlineSettings
   ) -> (then)
   SetVisibility(execute=@n2.else, self=[Get EditableText(), Get Border()], InVisibility=Hidden)
  ]}
]
variables [
 Context Menu Item Clicked {}
 File Path {type=string, default=""}
]
widgets {
 info {inheritedWidgetCount=0, parentClass=&EditorUtilityWidget, rootWidgetClass=&HorizontalBox, widgetCount=4}
 widgets [
  HorizontalBox_2 {widgetClassPath=&HorizontalBox, widget=HorizontalBox_2}
  Text {
   widgetClassPath=&TextBlock
   bIsVariable=true
   parent=HorizontalBox_2
   widget=Text
   properties {
    text=Animation Type
    colorAndOpacity {colorUseRule=UseColor_Specified, specifiedColor=<rgba 0.5451 0.5451 0.5451 1>}
    font {
     bForceMonospaced=false
     bMaterialIsStencil=false
     fontMaterial=null
     fontObject=&Roboto
     letterSpacing=0
     monospacedWidth=1
     size=10
     skewAmount=0
     typefaceFontName=Regular
     outlineSettings {
      bApplyOutlineToDropShadows=false
      bMiteredCorners=false
      bSeparateFillAlpha=false
      outlineMaterial=null
      outlineSize=0
      outlineColor=<rgba 0 0 0 1>
     }
    }
   }
   slot {class=HorizontalBoxSlot, properties {verticalAlignment=VAlign_Center, padding=<ltrb 14 3 20 3>}}
  }
  Border {
   widgetClassPath=&Border
   bIsVariable=true
   parent=HorizontalBox_2
   widget=Border
   properties {
    background {
     bIsDynamicallyLoaded=false
     drawAs=RoundedBox
     imageType=NoImage
     mirroring=NoMirror
     resourceName=None
     resourceObject=null
     tiling=NoTile
     imageSize=<32 32>
     margin=<ltrb 0 0 0 0>
     outlineSettings {
      bUseBrushTransparency=false
      roundingType=FixedRadius
      width=0
      color {colorUseRule=UseColor_Specified, specifiedColor=<rgba 0 0 0 0>}
      cornerRadii=<3 3 3 3>
     }
     tintColor {colorUseRule=UseColor_Specified, specifiedColor=<rgba 1 1 1 1>}
     uVRegion {bIsValid=false, max=<0 0>, min=<0 0>}
    }
    brushColor=<rgba 0.0048 0.0048 0.0048 1>
   }
   slot {class=HorizontalBoxSlot, properties {verticalAlignment=VAlign_Center, padding=<ltrb 0 0 14 0>, size {sizeRule=Fill, value=1}}}
  }
  EditableText {
   widgetClassPath=/Script/UMG.EditableText
   bIsVariable=true
   parent=Border
   widget=EditableText
   properties {
    text=Description
    widgetStyle {
     backgroundImageComposing {
      bIsDynamicallyLoaded=false
      drawAs=Border
      imageType=FullColor
      mirroring=NoMirror
      resourceName=../../../Engine/Content/Slate/Old/HyperlinkDotted.png
      resourceObject=null
      tiling=Both
      imageSize=<0 0>
      margin=<ltrb 0 0 0 0.1875>
      outlineSettings {
       bUseBrushTransparency=false
       roundingType=HalfHeightRadius
       width=0
       color {colorUseRule=UseColor_Specified, specifiedColor=<rgba 0 0 0 0>}
       cornerRadii=<0 0 0 0>
      }
      tintColor {colorUseRule=UseColor_Specified, specifiedColor=<rgba 1 1 1 1>}
      uVRegion {bIsValid=false, max=<0 0>, min=<0 0>}
     }
     backgroundImageSelected {
      bIsDynamicallyLoaded=false
      drawAs=Box
      imageType=FullColor
      mirroring=NoMirror
      resourceName=../../../Engine/Content/Slate/Common/EditableTextSelectionBackground.png
      resourceObject=null
      tiling=NoTile
      imageSize=<0 0>
      margin=<ltrb 0.25 0.25 0.25 0.25>
      outlineSettings {
       bUseBrushTransparency=false
       roundingType=HalfHeightRadius
       width=0
       color {colorUseRule=UseColor_Specified, specifiedColor=<rgba 0 0 0 0>}
       cornerRadii=<0 0 0 0>
      }
      tintColor {colorUseRule=UseColor_Specified, specifiedColor=<rgba 1 1 1 1>}
      uVRegion {bIsValid=false, max=<0 0>, min=<0 0>}
     }
     caretImage {
      bIsDynamicallyLoaded=false
      drawAs=Image
      imageType=FullColor
      mirroring=NoMirror
      resourceName=../../../Engine/Content/Slate/Old/White.png
      resourceObject=null
      tiling=NoTile
      imageSize=<16 16>
      margin=<ltrb 0 0 0 0>
      outlineSettings {
       bUseBrushTransparency=false
       roundingType=HalfHeightRadius
       width=0
       color {colorUseRule=UseColor_Specified, specifiedColor=<rgba 0 0 0 0>}
       cornerRadii=<0 0 0 0>
      }
      tintColor {colorUseRule=UseColor_Specified, specifiedColor=<rgba 1 1 1 1>}
      uVRegion {bIsValid=false, max=<0 0>, min=<0 0>}
     }
     colorAndOpacity {colorUseRule=UseColor_Specified, specifiedColor=<rgba 0.5451 0.5451 0.5451 1>}
     font {
      bForceMonospaced=false
      bMaterialIsStencil=false
      fontMaterial=null
      fontObject=&Roboto
      letterSpacing=0
      monospacedWidth=1
      size=10
      skewAmount=0
      typefaceFontName=Regular
      outlineSettings {
       bApplyOutlineToDropShadows=false
       bMiteredCorners=false
       bSeparateFillAlpha=false
       outlineMaterial=null
       outlineSize=0
       outlineColor=<rgba 0 0 0 1>
      }
     }
     hintColorAndOpacity {colorUseRule=UseColor_Foreground, specifiedColor=<rgba 1 0 1 1>}
     selectedBackgroundColor {colorUseRule=UseColor_Foreground, specifiedColor=<rgba 1 0 1 1>}
    }
   }
   slot {class=BorderSlot}
  }
 ]
}
