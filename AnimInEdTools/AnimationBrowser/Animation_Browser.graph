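# Editor utility widget for an Animation Browser that creates, lists and manages Anim Sequences, Level Sequences and montages for AnimatableCreature assets.
# String-map (Kwargs) entry points load a creature and resolve paths through ShotGrid or name formatting before creating assets.
# Sequencer helpers add spawnable actors with Control Rig tracks, linked animations and a stage that other bindings attach to.
# Inline Python handles sorting and timestamped sequence backups; a SaveGame tracks recent sequences.
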EditorUtilityWidgetBlueprint /AnimInEdTools/AnimationBrowser/Animation_Browser
graphs [
 AB_CreateNewAnimAndLevelSequence {nodes [
   n1 = AB_CreateNewAnimAndLevelSequence!() -> (then, Array:Array of Strings, Kwargs:Map of Strings to Strings) {is=title K2Node_FunctionEntry}
   > n2 = AB_Create_ParseKwargs!(execute, self, kwargs:Map of Strings to Strings=@n1.Kwargs) -> (then, Animatable Creature:Animatable Creature Object Reference, Asset:String, Feature:String)
   > SetSelectedOption!(
    execute
    self:"Searchable Combo Box (String) Object Reference"=Get Asset_ComboBox() -> (Asset_ComboBox:"Searchable Combo Box (String) Object Reference")
    Option:String=Get Asset!(self:Animatable Creature Object Reference=@n2) -> (Asset:String)
   ) -> (then)
   > n3 = GetEditorWorld(self=Get UnrealEditorSubsystem())
   > n4 = Cast To AnimatableCreature_ShotGrid!(execute, Object=@n2) -> (then, CastFailed, AsAnimatable Creature Shot Grid:Animatable Creature Shot Grid Object Reference) {is=title K2Node_DynamicCast}
   > n5 = Branch(Condition=Get SG Enabled!(self:Animatable Creature Shot Grid Object Reference=@n4) -> (is SG Enabled:bool) {is=title K2Node_CallFunction})
   > n6 = ResolveACPaths_SGEnabled!(execute, name:String=@n2.Asset, __WorldContext:Object Reference=@n3) -> (then, source_path:String, anim_path:String)
   > Set loc_lvlSeq(loc_lvlSeq:String=@n6) -> (Output_Get:String)
   > n7 = Set loc_animSeq(loc_animSeq:String=@n6.anim_path) -> (Output_Get:String)
   n8 = Get loc_animSeq() -> (loc_animSeq:String)
   n9 = Get loc_lvlSeq() -> (loc_lvlSeq:String)
   n10 = Branch(execute=[@n7.then, @n13.then], Condition=OR Boolean(A=IsEmpty(InString=@n9), B=IsEmpty(InString=@n8)))
   n11 = Create New Animation With Level Sequence!(
    execute=@n10.else
    AC:Animatable Creature Object Reference=@n2
    AnimatableCreatures:Array of Animatable Creature Object References=Get AnimatableCreature() -> (AnimatableCreature:Array of Animatable Creature Object References)
    Anim Sequence Path:String=Get loc_animSeq()
    Level Sequence Path:String=Get loc_lvlSeq()
    Project:BP Animation Browser Project Object Reference=Get Current Project() -> (Current Project:BP Animation Browser Project Object Reference)
    Save Game:SG Animation Browser Object Reference=Get Save Game() -> (Save Game:SG Animation Browser Object Reference)
    __WorldContext:Object Reference=@n3
   ) -> (then, LevelSequence:Level Sequence Object Reference, AnimSequence:Anim Sequence Object Reference)
   > Refresh Assets()
   > Return Node!(execute, LevelSequence:Level Sequence Object Reference=@n11, AnimSequence:Anim Sequence Object Reference=@n11.AnimSequence)
   n12 = FormatAnimationPathsFromAsset!(
    execute=[@n4.CastFailed, @n5.else]
    Asset Name:String=@n2.Asset
    AC File:Animatable Creature Object Reference=@n2
    Feature:String=@n2.Feature
    __WorldContext:Object Reference=@n3
   ) -> (then, Source Path:String, Level Sequence Path:String, Animation Path:String)
   > Set loc_lvlSeq(loc_lvlSeq=@n12.Level Sequence Path)
   > n13 = Set loc_animSeq(loc_animSeq=@n12.Animation Path)
  ]}
 AB_CreateNewAnimMontage_Lookup {nodes [
   n1 = AB_CreateNewAnimMontage_Lookup!() -> (then, Array:Array of Strings, Kwargs:Map of Strings to Strings) {is=title K2Node_FunctionEntry}
   n2 = Find(TargetMap=@n1.Kwargs, Key=AnimatableCreature)
   > n3 = Branch(Condition=@n2.ReturnValue)
   > n4 = LoadAsset(AssetPath=@n2)
   > n5 = Cast To AnimatableCreature!(execute, Object=@n4) -> (then, CastFailed, AsAnimatable Creature:Animatable Creature Object Reference) {is=title K2Node_DynamicCast}
   > n6 = Is Valid(InputObject=@n5)
   PrintString(execute=@n6.Is Not Valid, InString="Error: Found invalid AC file -- unable to load!")
   n7 = Find(TargetMap=@n1.Kwargs, Key=Asset)
   n8 = Find(TargetMap=@n1.Kwargs, Key=Subfolder)
   n9 = Find(TargetMap=@n1.Kwargs, Key=Plugin)
   n10 = Find(TargetMap=@n1.Kwargs, Key=Feature)
   n11 = Find(TargetMap=@n1.Kwargs, Key=Variant)
   n12 = Branch(
    execute=@n6.Is Valid
    Condition=AND Boolean(A=@n7.ReturnValue, B=OR Boolean!(A=@n8.ReturnValue, B=@n10.ReturnValue, C:bool=@n11.ReturnValue, D:bool=@n9.ReturnValue) -> (ReturnValue))
   )
   > n13 = Sequence()
   > n14 = Branch(Condition=@n8.ReturnValue)
   > Set local_feature(local_feature:String=@n8) -> (Output_Get:String)
   n15 = Branch(execute=@n14.else, Condition=@n10.ReturnValue)
   > Set local_feature(local_feature=@n10)
   n16 = Branch(execute=@n15.else, Condition=@n11.ReturnValue)
   > Set local_feature(local_feature=@n11)
   Set local_feature(execute=@n16.else, local_feature=@n9)
   PrintString(execute=@n13.then_1, InString=Append!(A="Info: Beginning Animation Creation for ~` ", B=@n2, C:String=" `~") -> (ReturnValue))
   > n17 = CreateNewMontage_Background!(execute, self, AC:Animatable Creature Object Reference=@n5, Asset:String=@n7, Feature:String=Get local_feature() -> (local_feature:String)) -> (then, AnimMontage:Anim Montage Object Reference)
   > Return Node!(execute, AnimMontage:Anim Montage Object Reference=@n17)
   PrintString(execute=@n12.else, InString="Error: Missing Kwargs ('Asset' or 'Feature') - These must be strings")
   PrintString(execute=@n3.else, InString="Error: Missing Kwargs ('AnimatableCreature') - This should be the package path to your AC file")
  ]}
 AB_CreateNewAnimSequence {nodes [
   n1 = AB_CreateNewAnimSequence!() -> (then, Array:Array of Strings, Kwargs:Map of Strings to Strings) {is=title K2Node_FunctionEntry}
   > n2 = AB_Create_ParseKwargs(kwargs=@n1.Kwargs)
   > SetSelectedOption(self=Get Asset_ComboBox(), Option=Get Asset(self=@n2))
   > n3 = GetEditorWorld(self=Get UnrealEditorSubsystem())
   > n4 = Cast To AnimatableCreature_ShotGrid(Object=@n2)
   > n5 = Branch(Condition=Get SG Enabled(self=@n4))
   > n6 = ResolveACPaths_SGEnabled(name=@n2.Asset, __WorldContext=@n3)
   > Set loc_lvlSeq(loc_lvlSeq=@n6)
   > Set loc_animSeq(loc_animSeq=@n6.anim_path)
   > n7 = PrintString(InString="   ~~ SG Enabled AC File Found ~~")
   n8 = Get loc_animSeq()
   n9 = Get loc_lvlSeq()
   n10 = Branch(execute=[@n7.then, @n13.then], Condition=OR Boolean!(A=IsEmpty(InString=@n9), B=IsEmpty(InString=@n8)) -> (ReturnValue))
   > PrintString(InString=Append!(A="Error: Failed to resolve paths for asset - ", B=@n2.Asset) -> (ReturnValue))
   n11 = Create New Anim Sequence!(
    execute=@n10.else
    AC:Animatable Creature Object Reference=@n2
    Anim Sequence Path:String=Get loc_animSeq()
    Level Sequence Path:String=Get loc_lvlSeq()
    Project:BP Animation Browser Project Object Reference=Get Current Project()
    Save Game:SG Animation Browser Object Reference=Get Save Game()
    __WorldContext:Object Reference=@n3
   ) -> (then, LevelSequence:Level Sequence Object Reference, AnimSequence:Anim Sequence Object Reference)
   > Refresh Assets()
   > Return Node!(execute, LevelSequence:Level Sequence Object Reference=@n11, AnimSequence:Anim Sequence Object Reference=@n11.AnimSequence)
   n12 = FormatAnimationPathsFromAsset(execute=[@n4.CastFailed, @n5.else], Asset Name=@n2.Asset, AC File=@n2, Feature=@n2.Feature, __WorldContext=@n3)
   > Set loc_lvlSeq(loc_lvlSeq=@n12.Level Sequence Path)
   > n13 = Set loc_animSeq(loc_animSeq=@n12.Animation Path)
  ]}
 AB_Create_ParseKwargs {nodes [
   AB_Create_ParseKwargs!() -> (then, kwargs:Map of Strings to Strings) {is=title K2Node_FunctionEntry}
   n1 = Get kwargs() -> (kwargs:Map of Strings to Strings)
   n2 = Find(TargetMap=@n1, Key=AnimatableCreature)
   > n3 = Branch(Condition=@n2.ReturnValue)
   > n4 = LoadAsset(AssetPath=@n2)
   > n5 = Cast To AnimatableCreature(Object=@n4)
   > n6 = Is Valid(InputObject=@n5)
   PrintString(execute=[@n5.CastFailed, @n6.Is Not Valid], InString="Error: Invalid AC file found -- unable to load!")
   n7 = Find(TargetMap=@n1, Key=Feature)
   n8 = Find(TargetMap=@n1, Key=Asset)
   n9 = Branch(execute=@n6.Is Valid, Condition=AND Boolean(A=@n8.ReturnValue, B=@n7.ReturnValue))
   > Return Node!(execute, Animatable Creature:Animatable Creature Object Reference=@n5, Asset:String=@n8, Feature:String=@n7)
   PrintString(execute=@n9.else, InString="Error: Missing Kwargs ('Asset' or 'Feature') - These must be strings")
   PrintString(execute=@n3.else, InString="Error: Missing Kwargs ('AnimatableCreature') - This should be the package path to your AC file")
  ]}
 AB_GetAnimationPaths {nodes [
   n1 = AB_GetAnimationPaths!() -> (then, AC File:Array of Animatable Creature Object References, Kwargs:Map of Strings to Strings) {is=title K2Node_FunctionEntry}
   n2 = "Get (a copy)"(Array:Array of Animatable Creature Object References=@n1) -> (Output:Animatable Creature Object Reference)
   > n3 = Is Valid(InputObject=@n2)
   PrintString(execute=@n3.Is Not Valid, InString=Error: Invalid AC file provided)
   n4 = Find(TargetMap=@n1.Kwargs, Key=Asset)
   n5 = Find(TargetMap=@n1.Kwargs, Key=Subfolder)
   n6 = Find(TargetMap=@n1.Kwargs, Key=Plugin)
   n7 = Find(TargetMap=@n1.Kwargs, Key=Feature)
   n8 = Find(TargetMap=@n1.Kwargs, Key=Variant)
   n9 = Branch(
    execute=@n3.Is Valid
    Condition=AND Boolean(A=@n4.ReturnValue, B=OR Boolean!(A=@n5.ReturnValue, B=@n7.ReturnValue, C:bool=@n8.ReturnValue, D:bool=@n6.ReturnValue) -> (ReturnValue))
   )
   > n10 = Sequence()
   > n11 = Branch(Condition=@n5.ReturnValue)
   > Set local_feature(local_feature=@n5)
   n12 = Branch(execute=@n11.else, Condition=@n7.ReturnValue)
   > Set local_feature(local_feature=@n7)
   n13 = Branch(execute=@n12.else, Condition=@n8.ReturnValue)
   > Set local_feature(local_feature=@n8)
   Set local_feature(execute=@n13.else, local_feature=@n6)
   PrintString(
    execute=@n10.then_1
    InString=Append!(A="Info: Beginning Animation Creation for ~` ", B=GetObjectName(Object=@n2), C:String=" `~") -> (ReturnValue)
   )
   > n14 = GetEditorWorld(self=Get UnrealEditorSubsystem())
   > n15 = FormatAnimationPathsFromAsset(Asset Name=@n4, AC File=@n2, Feature=Get local_feature(), __WorldContext=@n14)
   > Return Node!(execute, Source Path:String=@n15, Level Sequence Path:String=@n15.Level Sequence Path, Animation Path:String=@n15.Animation Path)
   PrintString(
    execute=@n9.else
    InString="Error: Missing Kwargs (Needs `Asset` and `Feature`) - These must be strings. Asset is the full ShotGrid name of the animation asset, Feature is the subfolder / 'plugin' name."
   )
  ]}
 AB_GetTokensFromFile {nodes [
   n1 = AB_GetTokensFromFile!() -> (then, Array:Array of Strings, Kwargs:Map of Strings to Strings) {is=title K2Node_FunctionEntry}
   n2 = Find(TargetMap=@n1.Kwargs, Key=AnimatableCreature)
   > n3 = Branch(Condition=@n2.ReturnValue)
   > n4 = LoadAsset(AssetPath=@n2)
   > n5 = Cast To AnimatableCreature(Object=@n4)
   > n6 = Is Valid(InputObject=@n5)
   PrintString(execute=@n6.Is Not Valid, InString="Error: Found invalid AC file -- unable to load!")
   n7 = For Each Loop(Exec=@n6.Is Valid, Array=Get Array() -> (Array:Array of Strings))
   Return Node!(execute=@n7.Completed, Names:Map of Strings to Strings=Get local_out() -> (local_out:Map of Strings to Strings))
   > n8 = Sequence()
   > n9 = Switch on String!(execute:Exec, Selection:String=@n7) -> (
    Default:Exec
    Subfolder:Exec
    Directions:Exec
    Direction:Exec
    Variation:Exec
    Variations:Exec
    Descriptor:Exec
    Descriptors:Exec
    AnimSet:Exec
   ) {pinNames [Subfolder,Directions,Direction,Variation,Variations,Descriptor,Descriptors,AnimSet]}
   > PrintString(InString=Append(A="Warning: Found invalid token ` ", B=@n7, C=" `"))
   n10 = Get local_sep() -> (local_sep:String)
   Set local_token(
    execute=@n9.AnimSet
    local_token:String=JoinStringArray(
     SourceArray=Get Anim Set!(self:Animatable Creature Object Reference=@n5) -> (Subfolders:Array of Strings) {is=title K2Node_CallFunction}
     Separator=@n10
    )
   ) -> (Output_Get:String)
   Set local_token(
    execute=@n9.Subfolder
    local_token=JoinStringArray(
     SourceArray=Get Subfolders!(self:Animatable Creature Object Reference=@n5) -> (Subfolders:Array of Strings) {is=title K2Node_CallFunction}
     Separator=@n10
    )
   )
   Set local_token(
    execute=[@n9.Direction, @n9.Direction, @n9.Directions, @n9.Directions]
    local_token=JoinStringArray(
     SourceArray=Get Direction!(self:Animatable Creature Object Reference=@n5) -> (Subfolders:Array of Strings) {is=title K2Node_CallFunction}
     Separator=@n10
    )
   )
   Set local_token(
    execute=[@n9.Variation, @n9.Variations]
    local_token=JoinStringArray(
     SourceArray=Get Variations!(self:Animatable Creature Object Reference=@n5) -> (Subfolders:Array of Strings) {is=title K2Node_CallFunction}
     Separator=@n10
    )
   )
   Set local_token(
    execute=[@n9.Descriptor, @n9.Descriptors]
    local_token=JoinStringArray(
     SourceArray=Get Descriptors!(self:Animatable Creature Object Reference=@n5) -> (Subfolders:Array of Strings) {is=title K2Node_CallFunction}
     Separator=@n10
    )
   )
   Add!(
    execute=@n8.then_1
    TargetMap:Map of Strings to Strings=Get local_out()
    Key:"String (by ref)"=@n7
    Value:"String (by ref)"=Get local_token() -> (local_token:String)
   ) -> (then) {is=title K2Node_CallFunction}
   PrintString(execute=@n3.else, InString="Error: Missing Kwargs ('AnimatableCreature') - This should be the package path to your AC file")
  ]}
 Add Actor {nodes [
   n1 = Add Actor!() -> (then, Animatable Creature:Animatable Creature Object Reference, Level Sequence:Level Sequence Object Reference, bIsLayeredControlRig:bool=~) {is=title K2Node_FunctionEntry}
   > n2 = LoadClassAsset_Blocking(AssetClass=Get Sequencer Actor!(self:Animatable Creature Object Reference=@n1) -> (Sequencer Actor:Object Soft Class Reference))
   > n3 = Load Asset Blocking!(
    execute
    InRef:"Control Rig Asset Soft Reference Structure (by ref)"=Get Control Rig!(self:Animatable Creature Object Reference=@n1) -> (Control Rig:Control Rig Asset Soft Reference Structure)
    bForceLoad:bool=true
   ) -> (then, ReturnValue:Control Rig Asset Strong Reference Structure)
   > n4 = Add Actor Base!(
    execute
    self
    Level Sequence:Level Sequence Object Reference=Get Level Sequence() -> (Level Sequence:Level Sequence Object Reference)
    bIsLayeredControlRig:bool=Get bIsLayeredControlRig() -> (bIsLayeredControlRig:bool)
    ControlRigClass:Control Rig Asset Strong Reference Structure=@n3
    Actor Class:Object Class Reference=@n2
   ) -> (then, New Binding:Movie Scene Binding Proxy Structure)
   > Return Node!(execute, New Binding:Movie Scene Binding Proxy Structure=@n4)
  ]}
 Add Actor And Control Rig {nodes [
   n1 = AnimatableCreature Find!(
    AnimatableCreatures:Array of Animatable Creature Object References=Get AnimatableCreature()
    Asset:String=Get Current Asset() -> (Current Asset:String)
    Variant:String=Get Current Variation() -> (Current Variation:String)
   ) -> (AnimatableCreature:Animatable Creature Object Reference, Exact Match:bool)
   n2 = LoadClassAsset_Blocking(execute=@n7.Is Valid, AssetClass=Get Sequencer Actor(self=@n1))
   n3 = Add Actor And Control Rig!() -> (then, Level:World Object Reference, LevelSequence:Level Sequence Object Reference) {is=title K2Node_FunctionEntry}
   n4 = AddSpawnableFromClass!(
    execute=@n5.then
    self:Level Sequence Editor Subsystem Object Reference=Get LevelSequenceEditorSubsystem() -> (ReturnValue:Level Sequence Editor Subsystem Object Reference) {is=title K2Node_GetEditorSubsystem}
    Sequence:Movie Scene Sequence Object Reference=@n3.LevelSequence
    ClassToSpawn:Object Class Reference=@n2
   ) -> (then, ReturnValue:Movie Scene Binding Proxy Structure)
   n5 = Load Asset Blocking(execute=@n2.then, InRef=Get Control Rig(self=@n1), bForceLoad=true)
   n6 = FindOrCreateControlRigTrackFromReference!(
    execute
    World:World Object Reference=@n3
    LevelSequence:Level Sequence Object Reference=@n3.LevelSequence
    InRef:"Control Rig Asset Strong Reference Structure (by ref)"=@n5
    InBinding:"Movie Scene Binding Proxy Structure (by ref)"=@n4
    bIsLayeredControlRig:bool=false
   ) -> (then, ReturnValue:Movie Scene Track Object Reference)
   Return Node!(execute=[@n6.then, @n4.then], Binding:Movie Scene Binding Proxy Structure=@n4)
   n7 = Is Valid(exec=@n3.then, InputObject=@n1)
  ]}
 Add Actor Base {nodes [
   n1 = Add Actor Base!() -> (
    then
    Level Sequence:Level Sequence Object Reference
    bIsLayeredControlRig:bool=~
    ControlRigClass:Control Rig Asset Strong Reference Structure
    Actor Class:Object Class Reference
   ) {is=title K2Node_FunctionEntry}
   > n2 = AddSpawnableFromClass(self=Get LevelSequenceEditorSubsystem(), Sequence=@n1, ClassToSpawn=@n1.Actor Class)
   > n3 = Branch(Condition=Is Valid Referfence!(InRef:"Control Rig Asset Strong Reference Structure (by ref)"=Get ControlRigClass() -> (ControlRigClass:Control Rig Asset Strong Reference Structure)) -> (ReturnValue:bool))
   > n4 = GetEditorWorld(self=Get UnrealEditorSubsystem())
   > FindOrCreateControlRigTrackFromReference(World=@n4, LevelSequence=Get Level Sequence(), InRef=Get ControlRigClass(), InBinding=@n2, bIsLayeredControlRig=Get bIsLayeredControlRig())
   > Return Node!(execute, New Binding:Movie Scene Binding Proxy Structure=@n2)
   Return Node(execute=@n3.else, New Binding=@n2)
  ]}
 Add Actor With Linked AnimSequence {nodes [
   n1 = Add Actor With Linked AnimSequence!() -> (then, AnimSequence Path:String, LevelSequence Path:String, Animatable Creature:Animatable Creature Object Reference) {is=title K2Node_FunctionEntry}
   > n2 = GetFocusedLevelSequence!(execute) -> (then, ReturnValue:Level Sequence Object Reference)
   > n3 = Add Actor!(
    execute
    self
    Animatable Creature:Animatable Creature Object Reference=@n1.Animatable Creature
    Level Sequence:Level Sequence Object Reference=@n2
    bIsLayeredControlRig:bool=false
   ) -> (then, New Binding:Movie Scene Binding Proxy Structure) {is=title K2Node_CallFunction}
   > n4 = Create Linked Anim Sequence!(
    execute
    self
    Level Sequence:Level Sequence Object Reference=@n2
    New AnimSequence Path:String=Get AnimSequence Path() -> (AnimSequence Path:String)
    Binding:Movie Scene Binding Proxy Structure=@n3
   ) -> (then, Anim Sequence:Anim Sequence Object Reference)
   > SaveLoadedAssets(AssetsToSave=Make Array!("[0]":Object Reference=@n4, "[1]":Object Reference=@n2) -> (Array:Array of Object References), bOnlyIfIsDirty=false)
  ]}
 Add Data To List {nodes [
   n1 = Add Data To List!() -> (
    then
    Anim Sequence:Soft Object Reference
    Level Sequence:Soft Object Reference
    Current User:String
    Display Subfolder:bool=~
    AnimMontages:Array of Soft Object References
   ) {is=title K2Node_FunctionEntry}
   > n2 = Construct Object from Class(Class=&/Object_AnimationBrowser.Object_AnimationBrowser_C, self=Self-Reference()) -> (ReturnValue:Object Animation Browser Object Reference)
   > Set Data!(
    execute
    Data:Struct Animation Brower Item Structure=Make Struct Animation Brower Item!(
     AnimSequence_31_#1:Soft Object Reference=@n1
     LevelSequence_32_#2:Soft Object Reference=@n1.Level Sequence
     CurrentUser_26_#3:String=@n1.Current User
    ) -> (Struct_AnimationBrower_Item:Struct Animation Brower Item Structure) {is=title K2Node_MakeStruct}
    self:Object Animation Browser Object Reference=@n2
   ) -> (then, Output_Get:Struct Animation Brower Item Structure)
   > Set AnimMontages!(execute, AnimMontages:Array of Soft Object References=@n1.AnimMontages, self:Object Animation Browser Object Reference=@n2) -> (then, Output_Get:Array of Soft Object References)
   > Set Display Subfolder!(execute, Display Subfolder:bool=@n1.Display Subfolder, self:Object Animation Browser Object Reference=@n2) -> (then, Output_Get:bool)
   > AddItem!(execute, self:List View Object Reference=Get ListView() -> (ListView:List View Object Reference), Item:Object Reference=@n2) -> (then)
   > Return Node!(execute, New Item:Object Animation Browser Object Reference=@n2)
  ]}
 Add LevelSequence to Recents {nodes [
   n1 = Add LevelSequence to Recents!() -> (then, LevelSequence:"String (by ref)") {is=title K2Node_FunctionEntry}
   n2 = Get Recent AnimSequences!(self:SG Animation Browser Object Reference=Get Save Game()) -> (Recent AnimSequences:Array of Strings)
   n3 = Find Item!(TargetArray=@n2, ItemToFind:"String (by ref)"=@n1) -> (ReturnValue)
   > n4 = Branch(Condition="integer > integer"(A=@n3, B=-1))
   > Remove Index!(execute, TargetArray=@n2, IndexToRemove:int=@n3) -> (then)
   > Insert!(execute, TargetArray=@n2, NewItem=Get LevelSequence() -> (LevelSequence:"String (by ref)"), Index:int=0) -> (then) {is=title K2Node_CallArrayFunction}
   Insert(execute=@n4.else, TargetArray=@n2, NewItem=Get LevelSequence(), Index=0)
   > Branch(Condition="integer > integer"(A=Length(TargetArray:Array of Strings=@n2), B=10))
   > Remove Index(TargetArray=@n2, IndexToRemove=10)
  ]}
 Add Stage Function {nodes [
   n1 = GetFocusedLevelSequence(execute=@n10.then)
   n2 = AddSpawnableFromClass(
    execute=@n47.then_1
    self=Get LevelSequenceEditorSubsystem()
    Sequence=@n1
    ClassToSpawn=/AnimInEdTools/AnimationBrowser/Assets/AnimationBrowserStage.AnimationBrowserStage_C
   )
   n3 = AddTrack!(
    execute=@n11.OutputPin
    InBinding:"Movie Scene Binding Proxy Structure (by ref)"=@n2
    TrackType:Movie Scene Track Class Reference=/Script/MovieSceneTracks.MovieScene3DTransformTrack
   ) -> (then, ReturnValue:Movie Scene Track Object Reference)
   > n4 = AddSection!(execute, Track:Movie Scene Track Object Reference=@n3) -> (then, ReturnValue:Movie Scene Section Object Reference)
   n5 = SetRange!(
    execute
    Section:Movie Scene Section Object Reference=@n4
    StartFrame:int=Get Start Frame() -> (Start Frame:int=0)
    EndFrame:int=Get End Frame() -> (End Frame:int=0)
   ) -> (then)
   SetBlendType!(execute=[@n4.then, @n5.then], self:Movie Scene Section Object Reference=@n4, InBlendType:EMovieSceneBlendType Enum=Absolute) -> (then)
   > n6 = GetChannelsByType!(
    execute
    Section:Movie Scene Section Object Reference=@n4
    ChannelType:Movie Scene Scripting Channel Class Reference=/Script/SequencerScripting.MovieSceneScriptingDoubleChannel
   ) -> (then, ReturnValue:Array of Movie Scene Scripting Channel Object References)
   > n7 = For Each Loop(Array:Array of Movie Scene Scripting Channel Object References=@n6) -> (Array Element:Movie Scene Scripting Channel Object Reference)
   > n8 = Cast To MovieSceneScriptingDoubleChannel!(execute, Object=@n7) -> (then, CastFailed, AsMovie Scene Scripting Double Channel:Movie Scene Scripting Double Channel Object Reference) {is=title K2Node_DynamicCast}
   n9 = Get!() -> (
    Transform_Location_X:double
    Transform_Location_Y:double
    Transform_Location_Z:double
    Transform_Rotation_Roll:float
    Transform_Rotation_Pitch:float
    Transform_Rotation_Yaw:float
    Transform_Scale_X:double
    Transform_Scale_Y:double
    Transform_Scale_Z:double
   )
   > "Set Default (Double)"!(
    execute
    self:Movie Scene Scripting Double Channel Object Reference=@n8
    InDefaultValue:double=Select!(
     Option 0:double=@n9
     Option 1:double=@n9.Transform_Location_Y
     Option 2:double=@n9.Transform_Location_Z
     Option 3:double=@n9.Transform_Rotation_Roll
     Option 4:double=@n9.Transform_Rotation_Pitch
     Option 5:double=@n9.Transform_Rotation_Yaw
     Option 6:double=@n9.Transform_Scale_X
     Option 7:double=@n9.Transform_Scale_Y
     Option 8:double=@n9.Transform_Scale_Z
     Index:int=@n7.Array Index
    ) -> (ReturnValue:double)
   ) -> (then) {is=title K2Node_CallFunction}
   Add Stage Function!() -> (then, Transform:Transform) {is=title K2Node_FunctionEntry}
   > Branch(Condition=InEditor!(WorldContextObject:Object Reference) -> (ReturnValue:bool))
   > n10 = Set Stage Exists(Stage Exists:bool=false) -> (Output_Get:bool)
   n11 = Reroute Node()
   n12 = GetBindings!(execute=@n47.then_3, Sequence:Movie Scene Sequence Object Reference=@n1) -> (then, ReturnValue:Array of Movie Scene Binding Proxy Structures)
   > n13 = For Each Loop(Array:Array of Movie Scene Binding Proxy Structures=@n12) -> (Array Element:Movie Scene Binding Proxy Structure)
   n14 = AddTrack(execute=@n53.then, InBinding=@n13, TrackType=&MovieScene3DTransformTrack)
   > n15 = AddSection(Track=@n14)
   SetRange(Section=@n15, StartFrame=Get Start Frame() -> (Start Frame=0), EndFrame=Get End Frame() -> (End Frame=0))
   SetBlendType(execute=@n40.then, self=@n15, InBlendType=Absolute)
   n16 = GetBindingID!(execute=@n19.then, Sequence:Movie Scene Sequence Object Reference=@n1, InBinding:"Movie Scene Binding Proxy Structure (by ref)"=@n2) -> (then, ReturnValue:Movie Scene Object Binding ID Structure)
   n17 = AddTrack(execute=@n49.then, InBinding=@n13, TrackType=/Script/MovieSceneTracks.MovieScene3DAttachTrack)
   > n18 = AddSection(Track=@n17)
   n19 = Cast To MovieScene3DAttachSection!(execute=@n41.then, Object=@n18) -> (then, CastFailed, AsMovie Scene 3DAttach Section:Movie Scene 3DAttach Section Object Reference) {is=title K2Node_DynamicCast}
   SetConstraintBindingID!(
    execute=@n16.then
    self:Movie Scene 3DConstraint Section Object Reference=@n19
    InConstraintBindingID:"Movie Scene Object Binding ID Structure (by ref)"=@n16
   ) -> (then)
   n20 = GetBindings(execute=@n1.then, Sequence=@n1)
   > n21 = For Each Loop(Array=@n20)
   n22 = FindTracksByExactType!(
    execute=@n52.then
    InBinding:"Movie Scene Binding Proxy Structure (by ref)"=@n21
    TrackType:Movie Scene Track Class Reference=&MovieScene3DTransformTrack
   ) -> (then, ReturnValue:Array of Movie Scene 3DTransform Track Object References)
   > n23 = For Each Loop(Array:Array of Movie Scene 3DTransform Track Object References=@n22) -> (Array Element:Movie Scene 3DTransform Track Object Reference)
   > n24 = GetSections!(execute, Track:Movie Scene Track Object Reference=@n23) -> (then, ReturnValue:Array of Movie Scene Section Object References)
   > n25 = For Each Loop(Array:Array of Movie Scene Section Object References=@n24) -> (Array Element:Movie Scene Section Object Reference)
   > n26 = GetChannelsByType(Section=@n25, ChannelType=&MovieSceneScriptingDoubleChannel)
   > n27 = For Each Loop(Array:Array of Movie Scene Scripting Channel Object References=@n26) -> (Array Element:Movie Scene Scripting Channel Object Reference)
   > n28 = Cast To MovieSceneScriptingDoubleChannel(Object=@n27)
   n29 = Get()
   n30 = Select(
    Option 0=@n29
    Option 1=@n29.Transform_Location_Y
    Option 2=@n29.Transform_Location_Z
    Option 3=@n29.Transform_Rotation_Roll
    Option 4=@n29.Transform_Rotation_Pitch
    Option 5=@n29.Transform_Rotation_Yaw
    Option 6=@n29.Transform_Scale_X
    Option 7=@n29.Transform_Scale_Y
    Option 8=@n29.Transform_Scale_Z
    Index=@n27.Array Index
   )
   "Set Default (Double)"(execute=@n54.else, self=@n28, InDefaultValue=@n30)
   SetRange(Section=@n18, StartFrame=Get Start Frame() -> (Start Frame=0), EndFrame=Get End Frame() -> (End Frame=0))
   n31 = GetObjectTemplate!(execute=@n2.then, InBinding:"Movie Scene Binding Proxy Structure (by ref)"=@n2) -> (then, ReturnValue:Object Reference)
   > Cast To Actor!(execute, Object=@n31) -> (then, CastFailed, AsActor:Actor Object Reference) {is=title K2Node_DynamicCast}
   n32 = GetPlaybackEnd!(execute=@n43.then, Sequence:Movie Scene Sequence Object Reference=@n1) -> (then, ReturnValue:int=0)
   > Set End Frame(End Frame:int=@n32) -> (Output_Get:int=0)
   n33 = Break Movie Scene Binding Proxy!(MovieSceneBindingProxy:"Movie Scene Binding Proxy Structure (by ref)"=@n2) -> (BindingID:Guid Structure, Sequence:Movie Scene Sequence Object Reference) {is=title K2Node_BreakStruct}
   n34 = Break Movie Scene Binding Proxy(MovieSceneBindingProxy=@n13)
   n35 = FindTracksByType!(
    execute=@n48.then_0
    InBinding:"Movie Scene Binding Proxy Structure (by ref)"=@n13
    TrackType:Movie Scene Track Class Reference=&MovieScene3DAttachTrack
   ) -> (then, ReturnValue:Array of Movie Scene 3DAttach Track Object References)
   n36 = Is Empty!(TargetArray:Array of Movie Scene 3DAttach Track Object References=@n35) -> (ReturnValue:bool)
   n37 = FindTracksByType(execute=@n48.then_1, InBinding=@n13, TrackType=&MovieScene3DTransformTrack) -> (ReturnValue:Array of Movie Scene 3DTransform Track Object References)
   n38 = Is Empty(TargetArray:Array of Movie Scene 3DTransform Track Object References=@n37)
   n39 = "Get Default (Double)"!(self:Movie Scene Scripting Double Channel Object Reference=@n28) -> (ReturnValue:double) {is=title K2Node_CallFunction}
   SetStartFrameBounded!(execute=@n15.then, Section:Movie Scene Section Object Reference=@n15, bIsBounded:bool=false) -> (then)
   > n40 = SetEndFrameBounded!(execute, Section:Movie Scene Section Object Reference=@n15, bIsBounded:bool=false) -> (then)
   SetStartFrameBounded(execute=@n18.then, Section=@n18, bIsBounded=false)
   > n41 = SetEndFrameBounded(Section=@n18, bIsBounded=false)
   n42 = GetPlaybackStart!(execute=@n47.then_0, Sequence:Movie Scene Sequence Object Reference=@n1) -> (then, ReturnValue:int=0)
   > n43 = Set Start Frame(Start Frame:int=@n42) -> (Output_Get:int=0)
   n44 = GetObjectTemplate(execute=@n21.LoopBody, InBinding=@n21)
   n45 = Get Class(Object=@n44) -> (ReturnValue:Object Class Reference)
   n46 = OR Boolean!(A="Equal (Class)"(A=&AnimationBrowserStage_C, B=@n45), B="Equal (Class)"(A=@n45)) -> (ReturnValue)
   n47 = Sequence!(execute=@n51.else) -> (then_0, then_1, then_2:Exec, then_3:Exec)
   n48 = Sequence!(execute=@n50.then) -> (then_0, then_1)
   n49 = Branch(execute=@n35.then, Condition=@n36)
   n50 = Branch(
    execute=@n13.LoopBody
    Condition="Not Equal (Guid)"(A:"Guid Structure (by ref)"=@n33, B:"Guid Structure (by ref)"=@n34) -> (ReturnValue:bool=~) {is=title K2Node_PromotableOperator}
   )
   n51 = Branch(execute=@n21.Completed, Condition=Get Stage Exists() -> (Stage Exists:bool))
   Branch(execute=@n44.then, Condition=@n46)
   > n52 = Set Stage Exists(Stage Exists=true)
   n53 = Branch(execute=@n37.then, Condition=@n38)
   n54 = Branch(execute=@n28.then, Condition="Equal (Float)"(A=@n39, B=@n30))
  ]}
 AnimMontage Validate Input {nodes [
   n1 = AnimMontage Validate Input!() -> (
    then
    AnimMontage Name:String
    AnimSequence Path:String
    Montage Preset:Blank Montage Class Reference
    Binding Proxy:Movie Scene Binding Proxy Structure
    Widget:EUW Dialog Create Anim Montage Object Reference
   ) {is=title K2Node_FunctionEntry}
   > Set Widget Local(Widget Local:EUW Dialog Create Anim Montage Object Reference=@n1.Widget) -> (Output_Get:EUW Dialog Create Anim Montage Object Reference)
   n2 = Get AnimMontage Path From AnimSequence!(AnimSequence Path:String=@n1.AnimSequence Path, InBinding:"Movie Scene Binding Proxy Structure (by ref)"=@n1.Binding Proxy) -> (AnimMontage Path:String, Is Valid:bool) {is=title K2Node_CallFunction}
   > Set Montage Path(Montage Path:String=@n2) -> (Output_Get:String)
   > n3 = Branch(Condition=@n2.Is Valid)
   n4 = Get AnimMontage Name() -> (AnimMontage Name:String)
   n5 = Append(A=@n4, B=., C=@n4)
   n6 = Get Montage Path() -> (Montage Path:String)
   > n7 = DoesAssetExist(AssetPath=Combine!(InPaths:Array of Strings=Make Array("[0]":String=@n6, "[1]":String=@n5) -> (Array:Array of Strings)) -> (ReturnValue:String))
   > n8 = Branch(Condition=@n7)
   > Set Data is Valid!(
    execute
    self:EUW Dialog Create Anim Montage Object Reference=Get Widget Local() -> (Widget Local:EUW Dialog Create Anim Montage Object Reference)
    Is Valid:bool=false
    Message:String
    Error:String="File Already Exists!"
   ) -> (then) {is=title K2Node_CallFunction}
   Set Data is Valid(execute=@n8.else, self=Get Widget Local(), Is Valid=true, Message=Get Montage Path())
   Set Data is Valid(execute=@n3.else, self=Get Widget Local(), Is Valid=false, Error=AC File not found)
  ]}
 Animatable Creature Get Variations {nodes [
   n1 = Animatable Creature Get Variations!() -> (then, AnimatableCreatures:Array of Animatable Creature Object References, Asset Name:String, Add All Option:bool=true) {is=title K2Node_FunctionEntry}
   > Clear(TargetArray=Get Local Creatures() -> (Local Creatures:Array of Strings))
   > n2 = For Each Loop(Array:Array of Animatable Creature Object References=@n1) -> (Array Element:Animatable Creature Object Reference)
   n3 = Execute Python Script!(
    execute=@n2.Completed
    PythonScript="sorted_creatures = sorted(creatures)\r\n\r\nif 'Shared' in sorted_creatures:\r\n    sorted_creatures.insert(0, sorted_creatures.pop(sorted_creatures.index('Shared')))\r\n\r\nif len(creatures) > 1 and add_all:\r\n    sorted_creatures.insert(0, 'All')"
    creatures:Array of Strings=Get Local Creatures()
    add_all:bool=Get Add All Option() -> (Add All Option:bool)
   ) -> (then, ReturnValue, sorted_creatures:Array of Strings) {inputs [creatures,add_all], outputs [sorted_creatures]}
   > Return Node!(execute, Variations:Array of Strings=@n3.sorted_creatures)
   Branch(execute=@n2.LoopBody, Condition="Equal Exactly (String)"(A=Get Asset(self=@n2), B=Get Asset Name() -> (Asset Name:String)))
   > Add(TargetArray=Get Local Creatures(), NewItem=Get Variant!(self:Animatable Creature Object Reference=@n2) -> (Variant:String)) {is=title K2Node_CallArrayFunction}
  ]}
 Animatable Creature for AnimSequence {nodes [
   n1 = Animatable Creature for AnimSequence!() -> (then, AnimSequence Path:String, Asset Name:String) {is=title K2Node_FunctionEntry}
   > n2 = For Each Loop(Array:Array of Strings=Animatable Creature Get Variations!(
     AnimatableCreatures:Array of Animatable Creature Object References=Get AnimatableCreature()
     Asset Name:String=@n1.Asset Name
     Add All Option:bool=false
    ) -> (Variations:Array of Strings) {is=title K2Node_CallFunction}) -> (Array Element:String)
   Return Node!(execute=@n2.Completed, AnimatableCreature:Animatable Creature Object Reference)
   n3 = AnimatableCreature Find(AnimatableCreatures=Get AnimatableCreature(), Asset=Get Asset Name(), Variant=@n2)
   > Branch(Condition=StartsWith(SourceString=Get AnimSequence Path(), InPrefix=Get Animation Path!(self:Animatable Creature Object Reference=@n3) -> (Animation Path:String)))
   > Return Node(AnimatableCreature=@n3)
  ]}
 AnimatableCreature Find {nodes [
   n1 = AnimatableCreature Find!() -> (then, AnimatableCreatures:Array of Animatable Creature Object References, Asset:String, Variant:String) {is=title K2Node_FunctionEntry}
   > n2 = Branch(Condition="Equal Exactly (String)"(A=@n1.Variant, B=All))
   > Set Best Fit(Best Fit:Animatable Creature Object Reference) -> (Output_Get:Animatable Creature Object Reference)
   > n3 = For Each Loop with Break!(Exec:Exec, Array:Array of Animatable Creature Object References=@n1, Break:Exec) -> (LoopBody:Exec, Array Element:Animatable Creature Object Reference, Array Index:int, Completed:Exec)
   Return Node!(execute=@n3.Completed, AnimatableCreature=Get Best Fit() -> (Best Fit:Animatable Creature Object Reference), Exact Match:bool=false)
   > Branch(Condition="Equal Exactly (String)"(A=Get Asset(self=@n3), B=Get Asset!() -> (Asset)))
   n4 = Get Variant(self=@n3)
   > n5 = Branch(Condition=OR Boolean(A="Equal, Case Insensitive (String)"(A=@n4, B=Shared), B="Equal, Case Insensitive (String)"(A=@n4, B=Core)))
   > Return Node(AnimatableCreature=@n3, Exact Match=false)
   Set Best Fit(execute=@n5.else, Best Fit=@n3)
   n6 = For Each Loop(Exec=@n2.else, Array:Array of Animatable Creature Object References=@n1) -> (Array Element:Animatable Creature Object Reference)
   Return Node(execute=@n6.Completed, Exact Match=false)
   n7 = Get Variant(self=@n6)
   n8 = Get Variant!() -> (Variant)
   n9 = Get Asset!(self:Animatable Creature Object Reference=@n6) -> (Asset)
   n10 = Get Asset!() -> (Asset)
   > Branch(Condition=AND Boolean(A="Equal, Case Insensitive (String)"(A=@n10, B=@n9), B="Equal, Case Insensitive (String)"(A=@n7, B=@n8)))
   > Return Node(AnimatableCreature=@n6, Exact Match=true)
  ]}
 AnimatableCreature Get Categories {nodes [
   n1 = AnimatableCreature Get Categories!() -> (then, AnimatableCreatures:Array of Animatable Creature Object References) {is=title K2Node_FunctionEntry}
   > Clear(TargetArray=Get Local Categories() -> (Local Categories:Array of Strings))
   > n2 = For Each Loop(Array=@n1)
   n3 = Execute Python Script!(execute=@n2.Completed, PythonScript="sorted_creatures = sorted(creatures)", creatures=Get Local Categories()) -> (then, ReturnValue, sorted_creatures) {inputs [creatures], outputs [sorted_creatures]}
   > Return Node!(execute, Categories:Array of Strings=@n3.sorted_creatures)
   n4 = Get Asset!(self:Animatable Creature Object Reference=@n2) -> (Asset)
   n5 = Branch(execute=@n2.LoopBody, Condition=IsEmpty(InString=@n4))
   > Add Unique(TargetArray=Get Local Categories(), NewItem=Get Variant!(self:Animatable Creature Object Reference=@n2) -> (Variant))
   Add Unique(execute=@n5.else, TargetArray=Get Local Categories(), NewItem=@n4)
  ]}
 Backup Level Sequence {nodes [
   n1 = Backup Level Sequence!() -> (then, Level Sequence:Level Sequence Object Reference) {is=title K2Node_FunctionEntry}
   n2 = Get AutoSave Dir!() -> (AutoSave Dir:String, Exists:bool) {is=title K2Node_CallFunction}
   > Branch(Condition=@n2.Exists)
   n3 = Make Array("[0]"=@n2, "[1]"=Get Object Path String(Object=@n1))
   > Execute Python Script!(
    execute
    PythonScript="import os\r\nimport shutil\r\nfrom pathlib import Path\r\n\r\nn = int(os.path.getmtime(src))\r\ndst = f'{dst}_VRS_{n}.uasset'\r\n\r\nif not os.path.exists(dst):\r\n    ## mkdir\r\n    Path(dst).parent.mkdir(parents=True, exist_ok=True)\r\n\r\n    ## Copy File\r\n    shutil.copyfile(src, dst)"
    src:String=GetSystemPath!(Object:Object Reference=@n1) -> (ReturnValue:String)
    dst:String=GetBaseFilename(InPath=Combine(InPaths=@n3), bRemovePath=false)
   ) -> (then, ReturnValue) {inputs [src,dst]}
  ]}
 Context Menu - Browse To Anim Montage {nodes [
   Context Menu - Browse To Anim Montage!() -> (then, Context Item:Editor Utility Widget Object Reference, Slot Widget:Editor Utility Widget Object Reference) {is=title K2Node_FunctionEntry}
   > Clear(TargetArray=Get AnimMontage Paths() -> (AnimMontage Paths:Array of Strings))
   > n1 = Get Selected ListView Item!(execute, self) -> (then, Item:Object Animation Browser Object Reference) {is=title K2Node_CallFunction}
   n2 = Get AnimMontages!(self:Object Animation Browser Object Reference=@n1) -> (AnimMontages:Array of Soft Object References)
   > n3 = Branch(Condition=Is Empty(TargetArray:Array of Soft Object References=@n2))
   n4 = For Each Loop(Exec=@n3.else, Array:Array of Soft Object References=@n2) -> (Array Element:Soft Object Reference)
   SyncBrowserToObjects(execute=@n4.Completed, AssetPaths=Get AnimMontage Paths())
   Add(
    execute=@n4.LoopBody
    TargetArray=Get AnimMontage Paths()
    NewItem=To Soft Object Path!(SoftObjectReference:Soft Object Reference=@n4) -> (ReturnValue_PathString:String)
   )
  ]}
 Context Menu - Create Montage {nodes [
   Context Menu - Create Montage!() -> (then, Context Item:Editor Utility Widget Object Reference, Slot Widget:Editor Utility Widget Object Reference) {is=title K2Node_FunctionEntry}
   > Show!(
    execute
    self:EUW Context Menu Object Reference=Get EUW_ContextMenu() -> (EUW_ContextMenu:EUW Context Menu Object Reference)
    Show:bool=false
    Mouse Position:Vector 2D Structure
    Name:String
   ) -> (then)
   > n1 = Display Modal Dialog!(
    execute
    self
    Class:User Widget Class Reference=/AnimInEdTools/AnimationBrowser/Dialog/CreateAnimMontage/EUW_Dialog_CreateAnimMontage.EUW_Dialog_CreateAnimMontage_C
   ) -> (then, Modal Widget:EUW Modal Dialog Object Reference, New Widget:User Widget Object Reference)
   > n2 = Cast To EUW_Dialog_CreateAnimMontage!(execute, Object=@n1.New Widget) -> (then, CastFailed, AsEUW Dialog Create Anim Montage:EUW Dialog Create Anim Montage Object Reference) {is=title K2Node_DynamicCast}
   > Bind Event to Cancel Button Pressed(self:EUW Dialog Create Anim Montage Object Reference=@n2, Delegate=Create Event()) {is=title K2Node_AddDelegate}
   > Bind Event to Create Button Pressed(self:EUW Dialog Create Anim Montage Object Reference=@n2, Delegate=Create Event()) {is=title K2Node_AddDelegate}
   > Bind Event to Validate Data(self:EUW Dialog Create Anim Montage Object Reference=@n2, Delegate=Create Event()) {is=title K2Node_AddDelegate}
   > n3 = Get Selected ListView Item()
   > n4 = LoadAsset_Blocking(Asset=Break Struct Animation Brower Item!(Struct_AnimationBrower_Item:"Struct Animation Brower Item Structure (by ref)"=Get Data!(self:Object Animation Browser Object Reference=@n3) -> (Data:Struct Animation Brower Item Structure)) -> (
     AnimSequence_31_#1:Soft Object Reference
     LevelSequence_32_#2:Soft Object Reference
     CurrentUser_26_#3:String
    ) {is=title K2Node_BreakStruct}.LevelSequence_32_#2)
   > n5 = Cast To LevelSequence!(execute, Object=@n4) -> (then, CastFailed, AsLevel Sequence:Level Sequence Object Reference) {is=title K2Node_DynamicCast}
   > Set Level Sequence!(execute, self:EUW Dialog Create Anim Montage Object Reference=@n2, LevelSequence:Level Sequence Object Reference=@n5) -> (then) {is=title K2Node_CallFunction}
  ]}
 Context Menu - Open in Level {nodes [
   n1 = Context Menu - Open in Level!() -> (then, Context Item:Editor Utility Widget Object Reference, Slot Widget:Editor Utility Widget Object Reference) {is=title K2Node_FunctionEntry}
   > n2 = Cast To EUW_ContextMenu_Icon!(execute, Object=@n1.Slot Widget) -> (then, CastFailed, AsEUW Context Menu Icon:EUW Context Menu Icon Object Reference) {is=title K2Node_DynamicCast}
   > n3 = Get Selected ListView Item()
   n4 = Get String Data!(self:EUW Context Menu Icon Object Reference=@n2) -> (String Data:String)
   > PrintString(InString=Append!(A="Open In ", B=@n4) -> (ReturnValue))
   > Open Item!(execute, self, Item:Object Animation Browser Object Reference=@n3, Level Path:String=@n4) -> (then)
  ]}
 Context Menu - Open in Level Sequence Set Default Map {nodes [
   n1 = Context Menu - Open in Level Sequence Set Default Map!() -> (then, Context Item:Editor Utility Widget Object Reference, Slot Widget:Editor Utility Widget Object Reference) {is=title K2Node_FunctionEntry}
   > Cast To EUW_ContextMenu_Icon(Object=@n1.Slot Widget)
   > n2 = Get Selected ListView Item()
   > n3 = LoadAsset(AssetPath="To String (SoftObjectReference)"!(SoftObjectReference:"Soft Object Reference (by ref)"=Get!(self:Object Animation Browser Object Reference=@n2) -> (
      Data_AnimSequence_31_#1:Soft Object Reference
      Data_LevelSequence_32_#2:Soft Object Reference
      Data_CurrentUser_26_#3:String
     ).Data_LevelSequence_32_#2) -> (ReturnValue:String))
   > n4 = Cast To LevelSequence(Object=@n3)
   > n5 = Get Default Map for Level Sequence!(execute, self, Level Sequence:Level Sequence Object Reference=@n4) -> (then, Default Map:String) {is=title K2Node_CallFunction}
   > Branch(Condition="Not Equal (Integer)"(A=Len(S=@n5)))
  ]}
 Context Menu - Project Settings {nodes [
   n1 = Context Menu - Project Settings!() -> (then, Context Item:Editor Utility Widget Object Reference, Slot Widget:Editor Utility Widget Object Reference) {is=title K2Node_FunctionEntry}
   > n2 = Cast To EUW_ContextMenu_Icon(Object=@n1.Slot Widget)
   > n3 = For Each Loop(Array:Array of EU Animation Browser Project Options Object References=Get Active Project Options() -> (Active Project Options:Array of EU Animation Browser Project Options Object References)) -> (Array Element:EU Animation Browser Project Options Object Reference)
   > n4 = For Each Loop!(
    Exec
    Array:Array of S AB Context Menu Options Structures=Get Context Menu Options!(self:EU Animation Browser Project Options Object Reference=@n3) -> (Context Menu Options:Array of S AB Context Menu Options Structures)
   ) -> (
    LoopBody
    Array Element_Enabled_2_#4:bool=~
    Array Element_DisplayName_6_#5:String
    Array Element_FunctionToCall_10_#6:String
    Array Element_Category_18_#7:String
    Array Element_Reference_15_#8:EUW Context Menu Icon Object Reference
    Array Element_Context_13_#9:Name
    Array Element_Icon_21_#10:Texture 2D Object Reference
    Array Index
    Completed
   )
   > n5 = Set local_ref(local_ref:EUW Context Menu Icon Object Reference=@n4.Array Element_Reference_15_#8) -> (Output_Get:EUW Context Menu Icon Object Reference)
   > Branch(Condition=@n4)
   > Branch(Condition="Equal (Object)"(A=@n5, B=@n2))
   > n6 = Execute Python Script!(
    execute
    PythonScript="result = obj.call_method(func)"
    func:String=@n4.Array Element_FunctionToCall_10_#6
    obj:EU Animation Browser Project Options Object Reference=@n3
   ) -> (then, ReturnValue) {inputs [func,obj]}
   > Branch(Condition=@n6)
   > Return Node!(execute)
  ]}
 Context Menu - Set Current Map as Default for Level Sequence {nodes [
   n1 = Context Menu - Set Current Map as Default for Level Sequence!() -> (then, Context Item:Editor Utility Widget Object Reference, Slot Widget:Editor Utility Widget Object Reference) {is=title K2Node_FunctionEntry}
   > Cast To EUW_ContextMenu_Icon(Object=@n1.Slot Widget)
   > n2 = Get Selected ListView Item()
   n3 = "To String (SoftObjectReference)"(SoftObjectReference=Get(self=@n2).Data_LevelSequence_32_#2)
   > n4 = LoadAsset(AssetPath=@n3)
   > n5 = Cast To LevelSequence(Object=@n4)
   > n6 = GetCurrentLevel!(execute, self:Level Editor Subsystem Object Reference=Get LevelEditorSubsystem()) -> (then, ReturnValue:Level Object Reference)
   > Set Default Map for Level Sequence!(
    execute
    self
    Level Sequence:Level Sequence Object Reference=@n5
    Path to Map:String=Split(SourceString=Get Object Path String(Object=@n6), InStr=:, SearchDir=FromEnd)
   ) -> (then) {is=title K2Node_CallFunction}
   RaiseScriptError!(execute=@n5.CastFailed, ErrorMessage:String=Append(A="[FAILED] to set Default Map for Level Sequence:  ", B=@n3)) -> (then)
  ]}
 Context Menu Helper - Level Sequence Has Default Map Set {nodes [
   Context Menu Helper - Level Sequence Has Default Map Set() {is=title K2Node_FunctionEntry}
   > n1 = Get Selected ListView Item()
   > n2 = LoadAsset(AssetPath="To String (SoftObjectReference)"(SoftObjectReference=Get(self=@n1).Data_LevelSequence_32_#2))
   > n3 = Cast To LevelSequence(Object=@n2)
   > n4 = Get Default Map for Level Sequence(Level Sequence=@n3)
   > Return Node!(execute, Level Sequence Has Default Map Set:bool="Not Equal (Integer)"(A=Len(S=@n4)))
  ]}
 Create AnimMontage {nodes [
   n1 = Create AnimMontage!() -> (
    then
    AnimMontage Name:String
    AnimSequence Path:String
    Montage Preset:Blank Montage Class Reference
    Binding Proxy:Movie Scene Binding Proxy Structure
   ) {is=title K2Node_FunctionEntry}
   > Remove Modal Dialog()
   > n2 = LoadAsset(AssetPath=@n1.AnimSequence Path)
   > n3 = Cast To AnimSequence!(execute, Object=@n2) -> (then, CastFailed, AsAnim Sequence:Anim Sequence Object Reference)
   > n4 = Construct Object from Class(Class=@n1.Montage Preset) -> (ReturnValue:Blank Montage Object Reference)
   > n5 = Create Montage!(
    execute
    self:Blank Montage Object Reference=@n4
    AssetName:String=Get AnimMontage Name()
    PackagePath:String=Get AnimMontage Path From AnimSequence(AnimSequence Path=Get AnimSequence Path(), InBinding=Get Binding Proxy() -> (Binding Proxy:Movie Scene Binding Proxy Structure))
    AnimSequence:Anim Sequence Object Reference=@n3
   ) -> (then, AnimMontage:Anim Montage Object Reference)
   > Is Valid(InputObject=@n5)
   > SaveLoadedAssets!(
    execute
    self:Editor Asset Subsystem Object Reference=Get EditorAssetSubsystem() -> (ReturnValue:Editor Asset Subsystem Object Reference) {is=title K2Node_GetEditorSubsystem}
    AssetsToSave=Make Array("[0]":Object Reference=@n3, "[1]":Object Reference=@n5) -> (Array:Array of Object References)
    bOnlyIfIsDirty=true
   ) -> (then, ReturnValue)
   > OpenEditorForAssets!(
    execute
    self:Asset Editor Subsystem Object Reference=Get AssetEditorSubsystem()
    Assets:Array of Object References=Make Array!("[0]"=@n5) -> (Array)
    OpenedMethod:EAssetTypeActivationOpenedMethod Enum=Edit
   ) -> (then, ReturnValue:bool)
   > Update List View!(
    execute
    self
    Current Variation:String=Get Current Variation()
    Animation Type:String=Get Current Animation Type() -> (Current Animation Type:String)
    Filter:String=Get Current Filter() -> (Current Filter:String)
    Current Category:String=Get Current Asset()
    Use Cache:bool=false
   ) -> (then)
  ]}
 Create AnimMontage Cancel Pressed {nodes [
   Create AnimMontage Cancel Pressed() {is=title K2Node_FunctionEntry}
   > Remove Modal Dialog()
  ]}
 Create Linked Anim Sequence {nodes [
   n1 = Create Linked Anim Sequence!() -> (then, Level Sequence:Level Sequence Object Reference, New AnimSequence Path:String, Binding:Movie Scene Binding Proxy Structure) {is=title K2Node_FunctionEntry}
   > n2 = Get Current Time!(execute, TimeUnit:EMovieSceneTimeUnit Enum=DisplayRate) -> (then, ReturnValue:Movie Scene Sequence Playback Params Structure) {is=title K2Node_CallFunction}
   n3 = Break Movie Scene Sequence Playback Params!(MovieSceneSequencePlaybackParams:"Movie Scene Sequence Playback Params Structure (by ref)"=@n2) -> (
    Frame_FrameNumber_Value:int
    Frame_SubFrame:float=~
    Time:float=~
    MarkedFrame:String
    Timecode:Timecode Structure
    PositionType:EMovieScenePositionType Enum
    UpdateMethod:EUpdatePositionMethod Enum
    bHasJumped:bool=~
   ) {is=title K2Node_BreakStruct}
   > Set Current Time!(
    execute
    PlaybackParams:Movie Scene Sequence Playback Params Structure=Make Movie Scene Sequence Playback Params!(
     Frame_FrameNumber_Value:int=int + int(A=@n3)
     Frame_SubFrame:float=0
     Time:float=0
     MarkedFrame:String
     Timecode:Timecode Structure=(Hours=0,Minutes=0,Seconds=0,Frames=0,Subframe=0,bDropFrameFormat=False)
     PositionType:EMovieScenePositionType Enum=Frame
     UpdateMethod:EUpdatePositionMethod Enum=Play
     bHasJumped:bool=False
    ) -> (MovieSceneSequencePlaybackParams:Movie Scene Sequence Playback Params Structure) {is=title K2Node_MakeStruct}
    TimeUnit:EMovieSceneTimeUnit Enum=DisplayRate
   ) -> (then) {is=title K2Node_CallFunction}
   > Set Current Time(PlaybackParams=@n2, TimeUnit=DisplayRate)
   > n4 = GetBindingID(Sequence=@n1, InBinding=@n1.Binding)
   > n5 = For Each Loop with Break(
    Array:Array of Object References=GetBoundObjects!(ObjectBinding:Movie Scene Object Binding ID Structure=@n4) -> (ReturnValue:Array of Object References)
    Break=@n15.then
   ) -> (Array Element:Object Reference)
   n6 = Is Valid(exec=@n5.Completed, InputObject=Get Skeleton() -> (Skeleton:Skeleton Object Reference))
   Return Node!(execute=@n6.Is Not Valid, Anim Sequence:Anim Sequence Object Reference)
   > Execute Python Script!(
    execute
    PythonScript="import unreal\r\n\r\n## Grab the Level Editor World\r\neditor_subsystem = unreal.get_editor_subsystem(unreal.UnrealEditorSubsystem)\r\nworld = editor_subsystem.get_editor_world()\r\n \r\n## Create animation sequence export options\r\nanim_seq_export_options = unreal.AnimSeqExportOption()\r\nanim_seq_export_options.export_transforms = True\r\nanim_seq_export_options.export_morph_targets = True\r\n \r\n## Get asset tools\r\nasset_tools = unreal.AssetToolsHelpers.get_asset_tools()\r\n\r\n## Create an anim sequence factory, assign it the target skeleton\r\nanim_sequence_factory = unreal.AnimSequenceFactory()\r\nanim_sequence_factory.target_skeleton = skeleton\r\nanim_sequence = unreal.AssetTools.create_asset(asset_tools, asset_name=asset_name, package_path=package_path, asset_class=unreal.AnimSequence, factory=anim_sequence_factory)  \r\n\r\n## Export\r\nunreal.ControlRigSequencerLibrary.export_anim_sequence_from_sequencer(anim_sequence, anim_seq_export_options, binding, True)"
    package_path:String=GetPath!(InPath:String=@n1.New AnimSequence Path) -> (ReturnValue:String)
    asset_name:String=GetBaseFilename(InPath=@n1.New AnimSequence Path, bRemovePath=true)
    level_sequence:Level Sequence Object Reference=@n1
    binding:Movie Scene Binding Proxy Structure=@n1.Binding
    skeleton:Skeleton Object Reference=Get Skeleton()
   ) -> (then, ReturnValue) {inputs [package_path, asset_name, level_sequence, binding, skeleton]}
   > n7 = LoadAsset(AssetPath=Get New AnimSequence Path() -> (New AnimSequence Path:String))
   > n8 = Cast To AnimSequence(Object=@n7)
   > n9 = Sequence()
   n10 = Get Compression Settings!(self:BP Animation Browser Project Object Reference=Get Current Project()) -> (Compression Settings:Anim Sequence Soft Object Reference)
   > Branch(Condition=IsValidSoftObjectReference(SoftObjectReference=@n10))
   > n11 = LoadAsset_Blocking(Asset=@n10)
   > n12 = Cast To AnimSequence(Object=@n11)
   > SetBoneCompressionSettings!(
    execute
    AnimationSequence:Anim Sequence Object Reference=@n8
    CompressionSettings:Anim Bone Compression Settings Object Reference=GetBoneCompressionSettings!(AnimationSequence:Anim Sequence Object Reference=@n12) -> (CompressionSettings:Anim Bone Compression Settings Object Reference)
   ) -> (then)
   > SetCurveCompressionSettings!(
    execute
    AnimationSequence:Anim Sequence Object Reference=@n8
    CompressionSettings:Anim Curve Compression Settings Object Reference=GetCurveCompressionSettings!(AnimationSequence:Anim Sequence Object Reference=@n12) -> (CompressionSettings:Anim Curve Compression Settings Object Reference)
   ) -> (then)
   > SetAnimationInterpolationType!(
    execute
    AnimationSequence:Anim Sequence Object Reference=@n8
    InterpolationType:EAnimInterpolationType Enum=GetAnimationInterpolationType!(AnimationSequence:Anim Sequence Object Reference=@n12) -> (InterpolationType:EAnimInterpolationType Enum=Linear)
   ) -> (then)
   Return Node(execute=@n9.then_1, Anim Sequence=@n8)
   n13 = Cast To Actor(execute=@n5.LoopBody, Object=@n5)
   n14 = GetComponentByClass!(self:Actor Object Reference=@n13, ComponentClass:Actor Component Class Reference=/Script/Engine.SkeletalMeshComponent) -> (ReturnValue:Skeletal Mesh Component Object Reference)
   > n15 = Set Skeleton(Skeleton:Skeleton Object Reference=Get Skeleton!(self:Skeletal Mesh Object Reference=Get SkeletalMesh!(self:Skinned Mesh Component Object Reference=@n14) -> (SkeletalMesh:Skeletal Mesh Object Reference)) -> (Skeleton)) -> (Output_Get:Skeleton Object Reference)
  ]}
 Create List Item Anim Sequence {nodes [
   Create List Item Anim Sequence() {is=title K2Node_FunctionEntry}
   > n1 = GetAssetsByClass!(
    execute
    self:Asset Registry Interface=GetAssetRegistry()
    ClassPathName:Top Level Asset Path Structure="Get Class Path Name (Top Level Asset Path)"!(Class:Object Class Reference=/Script/Engine.AnimSequence) -> (ReturnValue:Top Level Asset Path Structure) {is=title K2Node_CallFunction}
    bSearchSubClasses:bool=false
   ) -> (then, OutAssetData:Array of Asset Data Structures, ReturnValue:bool)
   > n2 = For Each Loop!(Exec, Array:Array of Asset Data Structures=@n1) -> (LoopBody, Array Element:Asset Data Structure, Array Index, Completed)
   n3 = Break Asset Data!(AssetData=@n2) -> (PackageName, AssetName)
   n4 = "To String (Name)"(InName=@n3.AssetName)
   n5 = Get Item Data!(self:Animation Browser List Item Object Reference=Get Current List Item() -> (Current List Item:Animation Browser List Item Object Reference)) -> (Item Data:Object Animation Browser Object Reference)
   n6 = "To String (SoftObjectReference)"(SoftObjectReference=Break Struct Animation Brower Item!(Struct_AnimationBrower_Item=Get Data(self=@n5)) -> (AnimSequence_31_#1))
   > Branch(Condition="Equal Exactly (String)"(A=@n4, B=Split(SourceString=@n6, InStr=., SearchDir=FromEnd).RightS))
   > n7 = Show Message Dialog(
    Title=MakeLiteralText(Value="LOCTEXT(\"Existing Anim Sequence Found\")")
    Message=Format Text!(
     Format="LOCTEXT(\"Duplicate asset detected! \\r\\nWould you like to load {x} instead of creating a new Anim Sequence?\\r\\n\\r\\nPath:: {path}\")"
     x:String=@n4
     path:Name=@n3
    ) -> (Result)
    MessageType=YesNoCancel
    DefaultValue=Yes
    MessageCategory=Warning
   )
   > n8 = Switch on EAppReturnType(Selection=@n7)
   Return Node!(execute=@n8.Cancel)
   > Set loc_flagged(loc_flagged:bool=true) -> (Output_Get:bool)
   n9 = Set loc_path(execute=@n8.Yes, loc_path:String="To String (Name)"(InName=@n3)) -> (Output_Get:String)
   n10 = Get loc_path() -> (loc_path:String)
   Add(execute=[@n9.then, @n14.then], TargetArray=Get Search AnimSequence Cache() -> (Search AnimSequence Cache:Array of Strings), NewItem=@n10)
   n11 = Get Item Data(self=Get Current List Item())
   > Set members in Struct Animation Brower Item!(
    execute:Exec
    StructRef:"Struct Animation Brower Item Structure (by ref)"=Get Data(self=@n11)
    AnimSequence_31_#1:Soft Object Reference=To Soft Object Reference!(SoftObjectPath_PathString:String=@n10) -> (ReturnValue)
   ) -> (then:Exec, StructOut:"Struct Animation Brower Item Structure (by ref)") {is=title K2Node_SetFieldsInStruct}
   > Reload Asset(self:Animation Browser List Item Object Reference=Get Current List Item())
   > Set Selected Item!(execute, self:List View Object Reference=Get ListView(), Item:Object Reference=@n11) -> (then) {is=title K2Node_CallFunction}
   > Return Node()
   n12 = Create New Anim Sequence(
    execute=@n2.Completed
    AC=Get Active AC File() -> (Active AC File:Animatable Creature Object Reference)
    Anim Sequence Path=@n6
    Project=Get Current Project()
    Save Game=Get Save Game()
   )
   > n13 = GetPathNameForLoadedAsset(LoadedAsset=@n12.AnimSequence)
   > n14 = Set loc_path(loc_path=@n13)
  ]}
 Create New Animation {nodes [
   n1 = Create New Animation!() -> (then, AnimSequence Path:String, LevelSequence Path:String, Animatable Creature:Animatable Creature Object Reference) {is=title K2Node_FunctionEntry}
   > Is Valid(InputObject=@n1.Animatable Creature)
   > Open Default Animation Map!(execute, self) -> (then, Map:World Object Reference)
   n2 = Get LevelSequence Path() -> (LevelSequence Path:String)
   > n3 = Execute Python Script!(
    execute
    PythonScript="import unreal\r\n\r\n## Get asset tools\r\nasset_tools = unreal.AssetToolsHelpers.get_asset_tools()\r\n \r\n## Creates a Level Sequence - /Game/Test_LS\r\nlevel_sequence = unreal.AssetTools.create_asset(asset_tools, asset_name = asset_name, package_path = package_path, asset_class = unreal.LevelSequence, factory = unreal.LevelSequenceFactoryNew())\r\n\r\n## Then open it in sequencer\r\nunreal.LevelSequenceEditorBlueprintLibrary.open_level_sequence(level_sequence)"
    asset_name=GetCleanFilename!(InPath:String=@n2) -> (ReturnValue:String)
    package_path=GetPath(InPath=@n2)
   ) -> (then, ReturnValue, level_sequence:Object Reference) {inputs [asset_name,package_path], outputs [level_sequence]}
   > Branch(Condition=@n3)
   > n4 = Cast To LevelSequence(Object=@n3.level_sequence)
   > n5 = Set Level Sequence(Level Sequence:Level Sequence Object Reference=@n4) -> (Output_Get:Level Sequence Object Reference) {is=title K2Node_VariableSet}
   > n6 = Add Actor(
    Animatable Creature=Get Animatable Creature() -> (Animatable Creature:Animatable Creature Object Reference)
    Level Sequence=@n5
    bIsLayeredControlRig=false
   )
   > Set New Binding(New Binding:Movie Scene Binding Proxy Structure=@n6) -> (Output_Get:Movie Scene Binding Proxy Structure)
   > n7 = Sequence!(execute) -> (then_0, then_1, then_2:Exec)
   n8 = AnimatableCreature Find(AnimatableCreatures=Get AnimatableCreature(), Asset=Get Current Asset(), Variant=Get Current Variation()) {is=title K2Node_CallFunction}
   > Is Valid(InputObject=@n8)
   n9 = Get Default Pose!(self:Animatable Creature Object Reference=@n8) -> (Default Pose:Control Rig Pose Asset Soft Object Reference)
   > Branch(Condition=IsValidSoftObjectReference(SoftObjectReference=@n9))
   n10 = Make Frame Number!(Value:int=1) -> (FrameNumber:Frame Number Structure) {is=title K2Node_MakeStruct}
   > Set Current Local Time!(
    execute
    PlaybackParams:Movie Scene Sequence Playback Params Structure=Make Movie Scene Sequence Playback Params!(
     Frame:Frame Time Structure=Make Frame Time!(FrameNumber:Frame Number Structure=@n10, SubFrame:float=0) -> (FrameTime:Frame Time Structure) {is=title K2Node_MakeStruct}
     Time=0
     MarkedFrame
     Timecode=(Hours=0,Minutes=0,Seconds=0,Frames=0,Subframe=0,bDropFrameFormat=False)
     PositionType=Frame
     UpdateMethod=Play
     bHasJumped=False
    ) -> (MovieSceneSequencePlaybackParams)
    TimeUnit:EMovieSceneTimeUnit Enum=DisplayRate
   ) -> (then) {is=title K2Node_CallFunction}
   n11 = Make Frame Number(Value=0)
   > Set Current Local Time(
    PlaybackParams=Make Movie Scene Sequence Playback Params(
     Frame=Make Frame Time(FrameNumber=@n11, SubFrame=0)
     Time=0
     Timecode=(Hours=0,Minutes=0,Seconds=0,Frames=0,Subframe=0,bDropFrameFormat=False)
     PositionType=Frame
     UpdateMethod=Play
     bHasJumped=False
    )
    TimeUnit=DisplayRate
   )
   > n12 = GetControlRigs!(execute, LevelSequence:Level Sequence Object Reference=Get Level Sequence()) -> (then, ReturnValue:Array of Control Rig Sequencer Binding Proxy Structures)
   > n13 = For Each Loop with Break(Array:Array of Control Rig Sequencer Binding Proxy Structures=@n12, Break=@n15.then) -> (Array Element:Control Rig Sequencer Binding Proxy Structure)
   n14 = Get New Binding() -> (New Binding:Movie Scene Binding Proxy Structure)
   > Branch(Condition="Equal (Guid)"(
     A:"Guid Structure (by ref)"=Break Movie Scene Binding Proxy(MovieSceneBindingProxy=@n14)
     B:"Guid Structure (by ref)"=Break Control Rig Sequencer Binding Proxy!(ControlRigSequencerBindingProxy:"Control Rig Sequencer Binding Proxy Structure (by ref)"=@n13) -> (
      Proxy_BindingID:Guid Structure
      Proxy_Sequence:Movie Scene Sequence Object Reference
      ControlRig:Control Rig Object Reference
      Track:Movie Scene Control Rig Parameter Track Object Reference
     ) {is=title K2Node_BreakStruct}
    ) -> (ReturnValue:bool=~) {is=title K2Node_PromotableOperator})
   > n15 = Paste Pose!(
    execute
    self
    Pose Asset:Soft Object Reference=@n9
    LevelSequence:Level Sequence Object Reference=Get Level Sequence()
    ControlRig Proxy Binding:"Control Rig Sequencer Binding Proxy Structure (by ref)"=@n13
   ) -> (then)
   Add Stage Function!(execute=@n7.then_2, self, Transform:Transform=MakeTransform(Rotation="0, 0, 0")) -> (then) {is=title K2Node_CallFunction}
   n16 = Create Linked Anim Sequence(execute=@n7.then_1, Level Sequence=Get Level Sequence(), New AnimSequence Path=Get AnimSequence Path(), Binding=Get New Binding()) {is=title K2Node_CallFunction}
   > SaveLoadedAssets!(execute, AssetsToSave=Make Array!("[0]"=Get Level Sequence(), "[1]":Object Reference=@n16) -> (Array), bOnlyIfIsDirty=false) -> (then, ReturnValue)
   > Set Selected Path!(
    execute
    Selected Path:String=GetSoftObjectPath!(Object=@n16) -> (ReturnValue_PathString:String)
    self:SG Animation Browser Object Reference=Get Save Game()
   ) -> (then, Output_Get:String)
   n17 = Get Save Game()
   > Set Selected Index!(execute, Selected Index:int=-1, self:SG Animation Browser Object Reference=@n17) -> (then, Output_Get:int=0)
   > Set Apply Default Pose!(execute, Apply Default Pose:bool=true, self:SG Animation Browser Object Reference=@n17) -> (then, Output_Get:bool)
   > SaveGameToSlot!(execute, SaveGameObject:Save Game Object Reference=@n17, SlotName:String=AnimationBrowser, UserIndex:int=0) -> (then, ReturnValue:bool)
  ]}
 Create New Level Sequence And Bake Animation To Rig {nodes [
   n1 = Create New Level Sequence And Bake Animation To Rig!() -> (
    then
    Asset Name:String
    Package Path:String
    bReduceKeys:bool=~
    Tolerance:float=~
    AnimSequence:Anim Sequence Base Object Reference
    ControlRigClass:Control Rig Asset Strong Reference Structure
    Actor Class:Object Class Reference
   ) {is=title K2Node_FunctionEntry}
   > n2 = Execute Python Script(
    PythonScript="import unreal\r\n\r\n## Get asset tools\r\nasset_tools = unreal.AssetToolsHelpers.get_asset_tools()\r\n \r\n## Creates a Level Sequence - &Test_LS\r\nlevel_sequence = unreal.AssetTools.create_asset(asset_tools, asset_name = asset_name, package_path = package_path, asset_class = unreal.LevelSequence, factory = unreal.LevelSequenceFactoryNew())\r\n\r\n## Then open it in sequencer\r\nunreal.LevelSequenceEditorBlueprintLibrary.open_level_sequence(level_sequence)"
    asset_name=@n1
    package_path=@n1.Package Path
   ) -> (level_sequence:Level Sequence Object Reference) {inputs [asset_name,package_path], outputs [level_sequence]}
   > n3 = "Set Level  Sequence"("Level  Sequence":Level Sequence Object Reference=@n2.level_sequence) -> (Output_Get:Level Sequence Object Reference)
   > SetViewRangeStart!(execute, InSequence:Movie Scene Sequence Object Reference=@n3, StartTimeInSeconds:double=-0.1) -> (then)
   n4 = float + float(
    A=Get SequenceLength!(self:Anim Sequence Base Object Reference=Get AnimSequence() -> (AnimSequence:Anim Sequence Base Object Reference)) -> (SequenceLength:float)
    B=0.1
   )
   > SetViewRangeEnd!(execute, InSequence:Movie Scene Sequence Object Reference=@n3, EndTimeInSeconds:double=@n4) -> (then)
   > SetWorkRangeStart!(execute, InSequence:Movie Scene Sequence Object Reference=@n3, StartTimeInSeconds:double=-0.1) -> (then)
   > SetWorkRangeEnd!(execute, InSequence:Movie Scene Sequence Object Reference=@n3, EndTimeInSeconds:double=@n4) -> (then)
   n5 = GetNumFrames!(AnimationSequenceBase:Anim Sequence Base Object Reference=Get AnimSequence()) -> (NumFrames:int=0)
   > SetPlaybackEnd!(execute, Sequence:Movie Scene Sequence Object Reference=@n3, EndFrame:int=@n5) -> (then)
   > n6 = Add Actor Base(
    Level Sequence="Get Level  Sequence"() -> ("Level  Sequence":Level Sequence Object Reference)
    bIsLayeredControlRig=false
    ControlRigClass=Get ControlRigClass()
    Actor Class=Get Actor Class() -> (Actor Class:Object Class Reference)
   ) {is=title K2Node_CallFunction}
   > n7 = Set Actor Binding(Actor Binding:Movie Scene Binding Proxy Structure=@n6) -> (Output_Get:Movie Scene Binding Proxy Structure)
   > n8 = AddTrack(InBinding=@n7, TrackType=/Script/MovieSceneTracks.MovieSceneSkeletalAnimationTrack)
   > Set Animation Track(Animation Track:Movie Scene Track Object Reference=@n8) -> (Output_Get:Movie Scene Track Object Reference)
   > n9 = AddSection(Track=@n8)
   > n10 = Cast To MovieSceneSkeletalAnimationSection!(execute, Object=@n9) -> (then, CastFailed, AsMovie Scene Skeletal Animation Section:Movie Scene Skeletal Animation Section Object Reference) {is=title K2Node_DynamicCast}
   > n11 = Set Params!(
    execute
    Params:Movie Scene Skeletal Animation Params Structure=Make Movie Scene Skeletal Animation Params!(
     Animation:Anim Sequence Base Object Reference=Get AnimSequence()
     FirstLoopStartFrameOffset:Frame Number Structure=(Value=0)
     StartFrameOffset:Frame Number Structure=(Value=0)
     EndFrameOffset:Frame Number Structure=(Value=0)
     PlayRate:Movie Scene Time Warp Variant Structure="EMovieSceneTimeWarpType::FixedPlayRate(PlayRate=1)"
     bReverse:bool=False
     SlotName:Name=DefaultSlot
     MirrorDataTable:Mirror Data Table Object Reference
     bSkipAnimNotifiers:bool=False
     bForceCustomMode:bool=False
     SwapRootBone:ESwapRootBone Enum=SwapRootBone_None
     bLinearPlaybackWhenScaled:bool=False
    ) -> (MovieSceneSkeletalAnimationParams:Movie Scene Skeletal Animation Params Structure) {is=title K2Node_MakeStruct}
    self:Movie Scene Skeletal Animation Section Object Reference=@n10
   ) -> (then, Output_Get:Movie Scene Skeletal Animation Params Structure)
   > SetEditorProperty(Object=@n10, PropertyName=Params, PropertyValue:Movie Scene Skeletal Animation Params Structure=@n11)
   > SetRange(Section=@n10, StartFrame=0, EndFrame=int + int(A=@n5))
   > n12 = Execute Python Script!(
    execute
    PythonScript="import unreal\r\n## Create animation sequence export options\r\nanim_seq_export_options = unreal.AnimSeqExportOption()\r\nanim_seq_export_options.export_transforms = True\r\nanim_seq_export_options.export_morph_targets = True"
   ) -> (then, ReturnValue, anim_seq_export_options:Anim Seq Export Option Object Reference) {outputs [anim_seq_export_options]}
   > n13 = GetEditorWorld(self=Get UnrealEditorSubsystem())
   > OpenLevelSequence!(execute, LevelSequence:Level Sequence Object Reference="Get Level  Sequence"()) -> (then, ReturnValue:bool)
   > ForceUpdate!(execute) -> (then)
   > BakeToControlRigFromReference!(
    execute
    World:World Object Reference=@n13
    LevelSequence:Level Sequence Object Reference="Get Level  Sequence"()
    InRef:"Control Rig Asset Strong Reference Structure (by ref)"=Get ControlRigClass()
    ExportOptions:Anim Seq Export Option Object Reference=@n12.anim_seq_export_options
    bReduceKeys:bool=Get bReduceKeys() -> (bReduceKeys:bool)
    Tolerance:float=0
    Binding:"Movie Scene Binding Proxy Structure (by ref)"=Get Actor Binding() -> (Actor Binding:Movie Scene Binding Proxy Structure)
    bResetControls:bool=true
   ) -> (then, ReturnValue:bool)
   > Execute Python Script!(
    execute
    PythonScript="import unreal\r\n\r\n## Grab the Level Editor World\r\neditor_subsystem = unreal.get_editor_subsystem(unreal.UnrealEditorSubsystem)\r\nworld = editor_subsystem.get_editor_world()\r\n \r\n## Create animation sequence export options\r\nanim_seq_export_options = unreal.AnimSeqExportOption()\r\nanim_seq_export_options.export_transforms = True\r\nanim_seq_export_options.export_morph_targets = True\r\n \r\n## Export\r\nunreal.ControlRigSequencerLibrary.export_anim_sequence_from_sequencer(anim_sequence, anim_seq_export_options, binding, True)\r\n"
    anim_sequence:Anim Sequence Base Object Reference=Get AnimSequence()
    level_sequence:Level Sequence Object Reference="Get Level  Sequence"()
    binding:Movie Scene Binding Proxy Structure=Get Actor Binding()
   ) -> (then, ReturnValue) {inputs [anim_sequence, level_sequence, binding]}
   > RemoveTrack!(
    execute
    Sequence:Movie Scene Sequence Object Reference="Get Level  Sequence"()
    Track:Movie Scene Track Object Reference=Get Animation Track() -> (Animation Track:Movie Scene Track Object Reference)
   ) -> (then, ReturnValue:bool)
   > SaveLoadedAssets(AssetsToSave=Make Array("[0]"="Get Level  Sequence"(), "[1]"=Get AnimSequence()), bOnlyIfIsDirty=false)
   > Update List View(
    Current Variation=Get Current Variation()
    Animation Type=Get Current Animation Type()
    Filter=Get Current Filter()
    Current Category=Get Current Asset()
    Use Cache=false
   )
  ]}
 CreateNewMontage_Background {nodes [
   n1 = CreateNewMontage_Background!() -> (then, AC:Animatable Creature Object Reference, Asset:String, Feature:String) {is=title K2Node_FunctionEntry}
   > Is Valid(InputObject=@n1)
   > SetSelectedOption(self=Get Asset_ComboBox(), Option=Get Asset(self=@n1))
   > n2 = FormatAnimationPathsFromAsset(Asset Name=@n1.Asset, AC File=@n1, Feature=@n1.Feature)
   > n3 = Branch(Condition=OR Boolean(A=IsEmpty(InString=@n2.Level Sequence Path), B=IsEmpty(InString=@n2.Animation Path)))
   > PrintString(InString=Append!(
     A="Error: Failed to create new animation asset for: "
     B=Get Asset!() -> (Asset)
     C:String=" - Reason: Invalid resolved output path. This is either a result of illegal naming or a failure in reverse token parsing. Please contact support - (Alex Mannion and Votch Levi)."
    ) -> (ReturnValue))
   Remove Modal Dialog(execute=@n3.else)
   n4 = Append(A=Get AnimMontage Path From AnimSequence(AnimSequence Path=@n2.Animation Path), B=/, C=@n1.Asset)
   > n5 = DoesAssetExist(AssetPath=@n4)
   > Branch(Condition=@n5)
   > n6 = LoadAsset(AssetPath=@n4)
   > Is Valid(InputObject=@n6)
   > n7 = Cast To AnimSequence(Object=@n6)
   > n8 = Construct Object from Class(Class=/AnimInEdTools/AnimationBrowser/Dialog/CreateAnimMontage/MontagePreset/BlankMontage.BlankMontage_C)
   > n9 = Create Montage(
    self=@n8
    AssetName=Append!(A=@n1.Asset, B=_M) -> (ReturnValue)
    PackagePath=Append!(A=@n4, B=_M., C:String=@n1.Asset, D:String=_M) -> (ReturnValue)
    AnimSequence=@n7
   )
   > Return Node!(execute, AnimMontage:Anim Montage Object Reference=@n9)
  ]}
 Delete Item {nodes [
   n1 = Delete Item!() -> (then, Data:Object Animation Browser Object Reference) {is=title K2Node_FunctionEntry}
   > n2 = Show Message Dialog(
    Title=MakeLiteralText(Value="LOCTEXT(\"Delete Files\")")
    Message=MakeLiteralText(Value="LOCTEXT(\"Are you sure you want to delete the AnimSequence and LevelSequence?\")")
    MessageType=YesNo
    DefaultValue=No
    MessageCategory=Warning
   )
   > n3 = Switch on EAppReturnType(Selection=@n2)
   Clear(execute=@n3.Yes, TargetArray:Array of Object References=Get Files To Delete() -> (Files To Delete:Array of Object References))
   > RemoveItem!(execute, self:List View Object Reference=Get ListView(), Item:Object Reference=@n1) -> (then)
   > n4 = Sequence()
   n5 = Get(self=Get Data!() -> (Data:Object Animation Browser Object Reference))
   > Branch(Condition=IsValidSoftObjectReference(SoftObjectReference=@n5))
   > n6 = LoadAsset_Blocking(Asset=@n5)
   > Is Valid(InputObject=@n6)
   > Add(TargetArray:Array of Object References=Get Files To Delete(), NewItem:"Object Reference (by ref)"=@n6)
   Branch(execute=@n4.then_1, Condition=IsValidSoftObjectReference(SoftObjectReference=@n5.Data_LevelSequence_32_#2))
   > n7 = LoadAsset_Blocking(Asset=@n5.Data_LevelSequence_32_#2)
   > Is Valid(InputObject=@n7)
   > Add(TargetArray=Get Files To Delete(), NewItem=@n7)
   Branch(execute=@n4.then_2, Condition=Is Not Empty!(TargetArray:Array of Object References=Get Files To Delete()) -> (ReturnValue:bool))
   > DeleteLoadedAssets!(execute, AssetsToDelete:Array of Object References=Get Files To Delete()) -> (then, ReturnValue:bool)
   DeleteAsset!(execute, AssetPathToDelete:String) -> (then, ReturnValue:bool)
  ]}
 Display Anim Montage Menu {nodes [
   n1 = Add Item!(
    execute
    self:EUW Context Menu Object Reference
    Event:"Delegate (by ref)"
    Slot Widget:User Widget Class Reference=/AnimInEdTools/ContextMenu/EUW_ContextMenu_Icon.EUW_ContextMenu_Icon_C
    Category:String=Restore File
   ) -> (then, Context Menu Item:EUW Context Menu Item Object Reference, Widget:Editor Utility Widget Object Reference)
   > Cast To EUW_Restore_ContextMenu_Item!(execute, Object=@n1.Widget) -> (then, CastFailed, AsEUW Restore Context Menu Item:EUW Restore Context Menu Item Object Reference) {is=title K2Node_DynamicCast}
   n2 = Display Anim Montage Menu!() -> (then, List View Object:Object Animation Browser Object Reference) {is=title K2Node_FunctionEntry}
   > Clear Context Menu(self:EUW Context Menu Object Reference=Get EUW_ContextMenu())
   > n3 = For Each Loop(Array:Array of Soft Object References=Get AnimMontages(self=@n2)) -> (Array Element:Soft Object Reference)
   n4 = GetMousePositionOnPlatform!(execute=@n3.Completed) -> (then, ReturnValue:Vector 2D Structure)
   > Show(self=Get EUW_ContextMenu(), Show=true, Mouse Position=@n4, Name=AnimMontage)
   > Set Selected Item(self=Get ListView(), Item=Get List View Object() -> (List View Object:Object Animation Browser Object Reference))
   n5 = BreakSoftObjectPath(InSoftObjectPath=To Soft Object Path!(SoftObjectReference=@n3) -> (ReturnValue:Soft Object Path Structure))
   n6 = Add Icon Item!(
    execute=@n3.LoopBody
    self:EUW Context Menu Object Reference=Get EUW_ContextMenu()
    Event:"Delegate (by ref)"=Create Event()
    Category:String=AnimMontages
    Name:String=GetBaseFilename(InPath=@n5, bRemovePath=true)
    Font Size:double=10
    Icon:Texture 2D Object Reference
   ) -> (then, Context Menu Icon:EUW Context Menu Icon Object Reference)
   > Set String Data!(execute, String Data:String=@n5, self:EUW Context Menu Icon Object Reference=@n6) -> (then, Output_Get:String)
  ]}
 Display Modal Dialog {nodes [
   n1 = Display Modal Dialog!() -> (then, Class:User Widget Class Reference) {is=title K2Node_FunctionEntry}
   > n2 = Create Widget!(
    execute:Exec
    WorldContextObject:Object Reference
    Class:User Widget Class Reference=&/EUW_ModalDialog.EUW_ModalDialog_C
    OwningPlayer:Player Controller Object Reference
   ) -> (then:Exec, ReturnValue:EUW Modal Dialog Object Reference)
   > n3 = Create Widget(Class=@n1) -> (ReturnValue:User Widget Object Reference)
   > AddChild(self=Get NamedSlot!(self:EUW Modal Dialog Object Reference=@n2) -> (NamedSlot:Named Slot Object Reference), Content=@n3)
   > AddChild(self=Get ModalLayer() -> (ModalLayer:Named Slot Object Reference), Content=@n2)
   > Modal Dialog Displayed!(execute, self, Modal:EUW Modal Dialog Object Reference=@n2, Widget:User Widget Object Reference=@n3) -> (then)
   > Return Node!(execute, Modal Widget:EUW Modal Dialog Object Reference=@n2, New Widget:User Widget Object Reference=@n3)
  ]}
 Duplicate Asset {nodes [
   n1 = Duplicate Asset!() -> (then, In Object:Object Reference, In Path:String) {is=title K2Node_FunctionEntry}
   > n2 = GetAssetTools()
   > n3 = DuplicateAsset!(
    execute
    self:Asset Tools Interface=@n2
    AssetName:String=GetCleanFilename(InPath=@n1.In Path)
    PackagePath:String=GetPath(InPath=@n1.In Path)
    OriginalObject:Object Reference=@n1
   ) -> (then, ReturnValue:Object Reference)
   > Return Node!(execute, New Object:Object Reference=@n3)
  ]}
 Duplicate Shot {nodes [
   n1 = Duplicate Shot!() -> (
    then
    Level Sequence:Soft Object Reference
    Anim Sequence:Soft Object Reference
    New Anim Sequence Path:String
    New Level Sequence Path:String
   ) {is=title K2Node_FunctionEntry}
   > n2 = Branch(Condition=IsValidSoftObjectReference(SoftObjectReference=@n1))
   > n3 = LoadAsset_Blocking(Asset=@n1)
   > n4 = Duplicate Asset!(execute, self, In Object:Object Reference=@n3, In Path:String=Get New Level Sequence Path() -> (New Level Sequence Path:String)) -> (then, New Object:Object Reference) {is=title K2Node_CallFunction}
   > n5 = Cast To LevelSequence(Object=@n4)
   > ClearLinkedAnimSequences!(execute, InLevelSequence:Level Sequence Object Reference=@n5) -> (then)
   > n6 = GetControlRigs(LevelSequence=@n5)
   > n7 = For Each Loop!(Exec, Array:Array of Control Rig Sequencer Binding Proxy Structures=@n6) -> (
    LoopBody
    Array Element_Proxy:Movie Scene Binding Proxy Structure
    Array Element_ControlRig:Control Rig Object Reference
    Array Element_Track:Movie Scene Control Rig Parameter Track Object Reference
    Array Index
    Completed
   )
   SaveLoadedAsset(execute=@n7.Completed, AssetToSave=@n5, bOnlyIfIsDirty=false)
   n8 = Get Current Asset()
   n9 = Get AnimatableCreature()
   n10 = Get Current Variation()
   n11 = Load Asset Blocking(execute=@n7.LoopBody, InRef=Get Control Rig(self=AnimatableCreature Find(AnimatableCreatures=@n9, Asset=@n8, Variant=@n10)), bForceLoad=true)
   > Branch(Condition="Equal (Control Rig Strong Reference)"(
     A:"Control Rig Asset Strong Reference Structure (by ref)"=@n11
     B:"Control Rig Asset Strong Reference Structure (by ref)"=GetAssetReference!(self:Control Rig Object Reference=@n7.Array Element_ControlRig) -> (ReturnValue:Control Rig Asset Strong Reference Structure)
    ) -> (ReturnValue:bool=~) {is=title K2Node_PromotableOperator})
   > OpenLevelSequence(LevelSequence=@n5)
   n12 = Get New Anim Sequence Path() -> (New Anim Sequence Path:String)
   > n13 = Create Linked Anim Sequence(Level Sequence=@n5, New AnimSequence Path=@n12, Binding=@n7)
   > n14 = Is Valid(InputObject=@n13)
   PrintString(execute=@n14.Is Not Valid, InString=Something went wrong when creating the Linked Anim Sequence)
   SaveAsset!(execute=@n14.Is Valid, AssetToSave:String=@n12, bOnlyIfIsDirty:bool=false) -> (then, ReturnValue:bool)
   n15 = LoadAsset_Blocking(execute=@n2.else, Asset=@n1.Anim Sequence)
   > Duplicate Asset(In Object=@n15, In Path=Get New Anim Sequence Path())
   SaveAsset(AssetToSave=Get New Anim Sequence Path(), bOnlyIfIsDirty=false)
  ]}
 EventGraph {nodes [
   "On Clicked (AddPropButton)"() {is=title K2Node_ComponentBoundEvent}
   > Show(self=Get EUW_ContextMenu(), Show=false)
   > n1 = Create Widget(Class=&/EUW_Animation_Dialog.EUW_Animation_Dialog_C) -> (ReturnValue:EUW Animation Dialog Object Reference)
   > AddChild(self=Get ModalLayer(), Content=@n1)
   > Set CurrentUser!(execute, CurrentUser:String=Get Current User() -> (Current User:String), self:EUW Animation Dialog Object Reference=@n1) -> (then, Output_Get:String)
   > Bind Event to Generate Animation Paths(self:EUW Animation Dialog Object Reference=@n1, Delegate=Create Event()) {is=title K2Node_AddDelegate}
   > Set Initial Settings!(
    execute
    self:EUW Animation Dialog Object Reference=@n1
    Project:BP Animation Browser Project Object Reference=Get Current Project()
    Selected Creature:Animatable Creature Object Reference
    Developer Mode:bool=Get Developer Mode() -> (Developer Mode:bool)
    Set Plugin:bool=true
    Animatable Creatures:Array of Animatable Creature Object References=Get Animatable Props!(Project:BP Animation Browser Project Object Reference=Get Current Project(), __WorldContext:Object Reference) -> (Animatable Creatures:Array of Animatable Creature Object References) {is=title K2Node_CallFunction}
    Show Project Dialog:bool=false
    Override Anim Path Validation:bool=false
   ) -> (then) {is=title K2Node_CallFunction}
   > Set Title and Button Text!(
    execute
    self:EUW Animation Dialog Object Reference=@n1
    Title:String=Add Prop to Scene
    Create Button:String=Add
    Description:String=Add an animated prop to your scene and create a new Linked AnimSequence
   ) -> (then) {is=title K2Node_CallFunction}
   "On Clicked (ClearSearchBox)"() {is=title K2Node_ComponentBoundEvent}
   > Set Text(self:Text Box Object Reference=Get SearchTextBox() -> (SearchTextBox:Editor Utility Editable Text Box Object Reference))
   > n2 = Set Current Filter(Current Filter:String) -> (Output_Get:String)
   RetriggerableDelay!(execute=[@n29.then, @n27.then, @n2.then], WorldContextObject:Object Reference, Duration=0.2) -> (then)
   > Update List View(
    Current Variation=Get Current Variation()
    Animation Type=Get Current Animation Type()
    Filter=Get Current Filter()
    Current Category=Get Current Asset()
    Use Cache=true
   )
   > Set Search!(execute, Search:Text=Get Text(self=Get SearchTextBox()), self:SG Animation Browser Object Reference=Get Save Game()) -> (then, Output_Get:Text)
   > Save Settings()
   "On Clicked (SettingsButton)"() {is=title K2Node_ComponentBoundEvent}
   n3 = Get EUW_ContextMenu()
   n4 = Get Name!(self:EUW Context Menu Object Reference=@n3) -> (Name:String)
   > n5 = Branch(Condition=AND Boolean(A=IsVisible(self=@n3), B="Equal Exactly (String)"(A=@n4, B=Settings)))
   > Show(self=Get EUW_ContextMenu(), Show=false)
   Clear Context Menu(execute=@n5.else, self=Get EUW_ContextMenu())
   > n6 = Sequence!(execute) -> (then_0, then_1, then_2, then_3:Exec, then_4:Exec, then_5:Exec)
   Load ContextMenu Additional Project Options!(execute=@n6.then_5, self, Context:Name=Settings) -> (then, Had Options:bool)
   n7 = Add Icon Item(
    execute=@n6.then_4
    self=Get EUW_ContextMenu()
    Event=Create Event()
    Category=Experimental
    Name=SelectString(A=Disable Shotgrid, B=Enable Shotgrid, bPickA=Get isShotgridEnabled() -> (isShotgridEnabled:bool))
    Font Size=10
   )
   > SetToolTipText!(
    execute
    self:Widget Object Reference=@n7
    InToolTipText:"Text (by ref)"=Format Text!(Format="LOCTEXT(\"Requires AnimatableCreature_Shotgrid files and configured Shotgrid AssetSchemas for your project (BR/Juno/Corgi/Husky/etc..).\\r\\nThis mode is currently only operational on Corgi.\")") -> (Result)
   ) -> (then)
   Add Icon Item(
    execute=@n6.then_1
    self=Get EUW_ContextMenu()
    Event=Create Event()
    Category=Developer
    Name=SelectString(A=Disable Developer Mode, B=Enable Developer Mode, bPickA=Get Developer Mode())
    Font Size=10
   )
   > Add Icon Item(self=Get EUW_ContextMenu(), Event=Create Event(), Category=Developer, Name=Browse to AC File, Font Size=10)
   > Add Icon Item(self=Get EUW_ContextMenu(), Event=Create Event(), Category=Developer, Name=Browse to Project, Font Size=10)
   n8 = GetCachedGeometry!(self:Widget Object Reference=Get SettingsButton() -> (SettingsButton:Button Object Reference)) -> (ReturnValue:"Geometry Structure (by ref)")
   Show(
    execute=@n6.then_3
    self=Get EUW_ContextMenu()
    Show=true
    Mouse Position="Local to Absolute (Coordinates)"!(Geometry:"Geometry Structure (by ref)"=@n8, LocalCoordinate_X:double=0, LocalCoordinate_Y:double=0) -> (ReturnValue:Vector 2D Structure)
    Name=Settings
   )
   n9 = For Each Loop!(
    Exec=@n6.then_2
    Array:Array of BP Animation Browser Project Object References=Get Projects!() -> (Projects:Array of BP Animation Browser Project Object References) {is=title K2Node_CallFunction}
   ) -> (LoopBody, Array Element:BP Animation Browser Project Object Reference, Array Index, Completed)
   > Add Icon Item(
    self=Get EUW_ContextMenu()
    Event=Create Event()
    Category=Projects
    Name=Get Name(self:BP Animation Browser Project Object Reference=@n9)
    Font Size=10
    Icon=/AnimInEdTools/ContextMenu/Icon/Icon_Edit
   )
   Add Icon Item(
    execute=@n6.then_0
    self=Get EUW_ContextMenu()
    Event=Create Event()
    Name=Refresh Assets
    Font Size=10
    Icon=/AnimInEdTools/AnimationBrowser/Icons/RefreshSpinner
   )
   "On Clicked (RestoreButton)"() {is=title K2Node_ComponentBoundEvent}
   n10 = Get EUW_ContextMenu()
   n11 = Get Name(self:EUW Context Menu Object Reference=@n10)
   > n12 = Branch(Condition=AND Boolean(A=IsVisible(self=@n10), B="Equal Exactly (String)"(A=@n11, B=Restore)))
   > Show(self=Get EUW_ContextMenu(), Show=false)
   n13 = GetFocusedLevelSequence(execute=@n12.else)
   > Backup Level Sequence!(execute, self, Level Sequence:Level Sequence Object Reference=@n13) -> (then) {is=title K2Node_CallFunction}
   > Clear Context Menu(self=Get EUW_ContextMenu())
   > n14 = Sequence!(execute) -> (then_0, then_1)
   > n15 = Branch(Condition=Is Empty(TargetArray:Array of Movie Scene Binding Proxy Structures=GetSelectedBindings!() -> (ReturnValue:Array of Movie Scene Binding Proxy Structures)))
   > Add Icon Item(
    self=Get EUW_ContextMenu()
    Event=CustomEvent_8!() -> (OutputDelegate, then, Context Item:Editor Utility Widget Object Reference, Slot Widget:Editor Utility Widget Object Reference) {is=title K2Node_CustomEvent}
    Category=Restore File
    Name=Please select an asset to restore
    Font Size=10
   )
   n16 = GetFocusedLevelSequence(execute=@n15.else)
   > n17 = Get Level Sequence Backups!(execute, self, Level Sequence:Level Sequence Object Reference=@n16) -> (then, Level Sequence Backups:Array of Strings, Base Name:String) {is=title K2Node_CallFunction}
   > n18 = Execute Python Script!(
    execute
    PythonScript="import os\r\nimport time\r\n\r\nsorted_files = sorted(files_to_sort, key=lambda x: os.path.getmtime(x), reverse=True)"
    files_to_sort:Array of Strings=@n17
   ) -> (then, ReturnValue, sorted_files:Array of Strings) {inputs [files_to_sort], outputs [sorted_files]}
   > n19 = For Each Loop(Array:Array of Strings=@n18.sorted_files) -> (Array Element:String)
   > n20 = Execute Python Script!(
    execute
    PythonScript="import os\r\nimport time\r\n\r\nauto_save = True\r\n\r\nname = os.path.basename(file_path)\r\nif name.startswith(f'{base_name}_Auto'):\r\n    n = os.path.getmtime(file_path)\r\n    t = time.localtime(n)\r\n    s = time.strftime(\"%x %I:%M %p\",t)\r\nelif name.startswith(f'{base_name}_VRS_'):\r\n    suffix = name.split('_VRS_')[-1]\r\n    n = int(suffix.split('.')[0])\r\n    t = time.localtime(n)\r\n    s = time.strftime(\"%x %I:%M %p\",t)\r\n    auto_save = False\r\nelse:\r\n    s = ''\r\n\r\n##if n.isnumeric():\r\n##    t = time.localtime(int(n))\r\n##    s = time.strftime(\"%x %X%p\",t)"
    file_path:String=@n19
    base_name:String=@n17.Base Name
   ) -> (then, ReturnValue, s:String, auto_save:bool=~) {inputs [file_path,base_name], outputs [s,auto_save]}
   > n21 = Branch(Condition=IsEmpty(InString=@n20.s))
   n22 = Restore File ContextMenu!() -> (OutputDelegate, then, Context Item:Editor Utility Widget Object Reference, Slot Widget:Editor Utility Widget Object Reference) {is=title K2Node_CustomEvent}
   n23 = Add Item(
    execute=@n21.else
    self=Get EUW_ContextMenu()
    Event=@n22
    Slot Widget=/AnimInEdTools/AnimationBrowser/Restore/EUW_Restore_ContextMenu_Item.EUW_Restore_ContextMenu_Item_C
    Category=Restore File
   )
   > n24 = Cast To EUW_Restore_ContextMenu_Item(Object=@n23.Widget)
   > Set Text(
    Text="To Text (String)"(InString=@n20.s)
    self:Text Object Reference=Get Text(self:EUW Restore Context Menu Item Object Reference=@n24) -> (Text:Text Object Reference)
   )
   > Set File Path!(execute, File Path:String=@n19, self:EUW Restore Context Menu Item Object Reference=@n24) -> (then, Output_Get:String)
   > Set Autosave!(execute, self:EUW Restore Context Menu Item Object Reference=@n24, Value:bool=@n20.auto_save) -> (then) {is=title K2Node_CallFunction}
   n25 = Get RestoreButton() -> (RestoreButton:Button Object Reference)
   Show(
    execute=@n14.then_1
    self=Get EUW_ContextMenu()
    Show=true
    Mouse Position="Local to Absolute (Coordinates)"(Geometry=GetCachedGeometry(self=@n25), LocalCoordinate_X=0, LocalCoordinate_Y=0)
    Name=Restore
   )
   n26 = "On Text Changed (SearchTextBox)"!() -> (OutputDelegate, then, Text:"Text (by ref)") {is=title K2Node_ComponentBoundEvent}
   > n27 = Set Current Filter(Current Filter="To String (Text)"(InText=@n26.Text))
   n28 = "On Text Committed (SearchTextBox)"!() -> (OutputDelegate, then, Text:"Text (by ref)", CommitMethod:ETextCommit Enum=Default) {is=title K2Node_ComponentBoundEvent}
   > n29 = Set Current Filter(Current Filter="To String (Text)"(InText=@n28.Text))
   "On Opening (Subfolder_Combobox)"() {is=title K2Node_ComponentBoundEvent}
   > n30 = Cast To AnimatableCreature_ShotGrid(Object=Get Active AC File())
   n31 = NOT Boolean(A=Is ComboBox All!(CB:"Searchable Combo Box (String) Object Reference"=Get Plugin_Combobox() -> (Plugin_Combobox:"Searchable Combo Box (String) Object Reference")) -> ("  ":bool))
   n32 = Get SG Enabled(self=@n30)
   n33 = Get isShotgridEnabled()
   Update Combo Box!(
    execute=[@n30.CastFailed, @n30.then]
    self
    Combo Box:"Searchable Combo Box (String) Object Reference"=Get Subfolder_Combobox() -> (Subfolder_Combobox:"Searchable Combo Box (String) Object Reference")
    Items:Array of Strings=Get Subfolders!() -> (Subfolders)
    Default Item:String=SelectString(
     B=Get Animation Type Default Item() -> (Animation Type Default Item:String)
     bPickA=AND Boolean!(A=@n32, B=@n33, C:bool=@n31) -> (ReturnValue)
    )
    Add Blank:bool=false
   ) -> (then)
   > Show(self=Get EUW_ContextMenu(), Show=false)
   n34 = "On Selection Changed (Subfolder_Combobox)"!() -> (OutputDelegate, then, SelectedItem:String, SelectionType:ESelectInfo Enum) {is=title K2Node_ComponentBoundEvent}
   > n35 = Sequence()
   n36 = GetSelectedOption!(self:"Searchable Combo Box (String) Object Reference"=Get Subfolder_Combobox()) -> (ReturnValue:String)
   n37 = Get Current Animation Type()
   > n38 = Branch(Condition=OR Boolean(A="Equal Exactly (String)"(A=@n36, B=@n37), B=IsEmpty(InString=@n34.SelectedItem)))
   Update List View(
    execute=@n38.else
    Current Variation=Get Current Variation()
    Animation Type=@n34.SelectedItem
    Filter=Get Current Filter()
    Current Category=Get Current Asset()
    Use Cache=false
   )
   n39 = Switch on ESelectInfo!(execute=@n35.then_1, Selection:ESelectInfo Enum=@n34.SelectionType) -> (OnKeyPress:Exec, OnNavigation:Exec, OnMouseClick:Exec, Direct:Exec) {is=title K2Node_SwitchEnum}
   Set Subfolder!(
    execute=[@n39.OnKeyPress, @n39.OnMouseClick, @n39.OnNavigation]
    Subfolder:String=@n34.SelectedItem
    self:SG Animation Browser Object Reference=Get Save Game()
   ) -> (then, Output_Get:String)
   > Save Settings()
   n40 = "On Selection Changed (Asset_ComboBox)"!() -> (OutputDelegate, then, SelectedItem:String, SelectionType:ESelectInfo Enum) {is=title K2Node_ComponentBoundEvent}
   > n41 = Sequence!(execute) -> (then_0, then_1, then_2:Exec)
   > n42 = Set Current Asset(Current Asset:String=@n40.SelectedItem) -> (Output_Get:String)
   > Set Visibility(Visibility=Visible, self=Get PluginHorizontalBox() -> (PluginHorizontalBox:Horizontal Box Object Reference))
   n43 = Animatable Creature Get Variations(AnimatableCreatures=Get AnimatableCreature(), Asset Name=@n42, Add All Option=true)
   > Update Combo Box(Combo Box=Get Plugin_Combobox(), Items=@n43, Add Blank=false)
   > Branch(Condition="integer > integer"(A=Length(TargetArray=@n43)))
   > SetSelectedOption(self=Get Plugin_Combobox(), Option="Get (a copy)"(Array:Array of Strings=@n43) -> (Output:String))
   > Update List View From UI()
   Update Subfolder List!(execute=@n41.then_1, self) -> (then, Updated:bool)
   n44 = Switch on ESelectInfo(execute=@n41.then_2, Selection=@n40.SelectionType)
   Set Asset!(execute=[@n44.OnKeyPress, @n44.OnMouseClick], Asset:String=Get Current Asset(), self:SG Animation Browser Object Reference=Get Save Game()) -> (then, Output_Get:String)
   > Save Settings()
   n45 = "On Selection Changed (Plugin_Combobox)"!() -> (OutputDelegate, then, SelectedItem:String, SelectionType:ESelectInfo Enum) {is=title K2Node_ComponentBoundEvent}
   > n46 = Switch on ESelectInfo(Selection=@n45.SelectionType)
   n47 = Sequence(execute=[@n46.OnKeyPress, @n46.OnMouseClick])
   > n48 = Branch(Condition="Equal Exactly (String)"(A=Get Current Variation(), B=@n45.SelectedItem))
   Update List View(
    execute=[@n48.else, @n46.Direct, @n46.OnNavigation]
    Current Variation=@n45.SelectedItem
    Animation Type=Get Current Animation Type()
    Filter=Get Current Filter()
    Current Category=Get Current Asset()
    Use Cache=false
   )
   > SetIsEnabled(self=Get AddAnimationButton() -> (AddAnimationButton:Button Object Reference), bInIsEnabled=true)
   > SetIsEnabled(self=Get Subfolder_Combobox(), bInIsEnabled=true)
   Update Subfolder List(execute=@n47.then_1)
   Set Variant!(execute=@n47.then_2, Variant:String=@n45.SelectedItem, self:SG Animation Browser Object Reference=Get Save Game()) -> (then, Output_Get:String)
   > Save Settings()
   n49 = "On Opening (Asset_ComboBox)"() {is=title K2Node_ComponentBoundEvent}
   Update Combo Box(
    execute=[@n49.then, @n138.then]
    Combo Box=Get Asset_ComboBox()
    Items=AnimatableCreature Get Categories!(AnimatableCreatures:Array of Animatable Creature Object References=Get AnimatableCreature()) -> (Categories:Array of Strings) {is=title K2Node_CallFunction}
    Add Blank=false
   )
   > Show(self=Get EUW_ContextMenu(), Show=false)
   "On Opening (Plugin_Combobox)"() {is=title K2Node_ComponentBoundEvent}
   > Update Combo Box(
    Combo Box=Get Plugin_Combobox()
    Items=Animatable Creature Get Variations(AnimatableCreatures=Get AnimatableCreature(), Asset Name=Get Current Asset(), Add All Option=true)
    Add Blank=false
   )
   > Show(self=Get EUW_ContextMenu(), Show=false)
   n50 = "On Item Double Clicked (ListView)"!() -> (OutputDelegate, then, Item:Object Reference) {is=title K2Node_ComponentBoundEvent}
   n51 = Get Current List Item()
   > n52 = Branch(Condition=AND Boolean!(A=Get isShotgridEnabled(), B=Get CanCreate!(self:Animation Browser List Item Object Reference=@n51) -> (CanCreate:bool)) -> (ReturnValue))
   > Create List Item Anim Sequence() {is=title K2Node_CallFunction}
   n53 = Cast To Object_AnimationBrowser!(execute=@n52.else, Object=@n50.Item) -> (then, CastFailed, AsObject Animation Browser:Object Animation Browser Object Reference) {is=title K2Node_DynamicCast}
   > Open Item(Item=@n53)
   "On Clicked (AddAnimationButton)"() {is=title K2Node_ComponentBoundEvent}
   > n54 = Sequence!(execute) -> (then_0, then_1)
   n55 = Create Widget(execute=[@n54.then_0, @n154.else], Class=&EUW_Animation_Dialog_C)
   > AddChild(self=Get ModalLayer(), Content=@n55)
   > Set CurrentUser(CurrentUser=Get Current User(), self=@n55)
   > Bind Event to Generate Animation Paths(self=@n55, Delegate=Create Event())
   > Modal Dialog Displayed(Widget=@n55)
   n56 = AnimatableCreature Find(
    AnimatableCreatures=Get AnimatableCreature()
    Asset=GetSelectedOption(self=Get Asset_ComboBox())
    Variant=GetSelectedOption(self=Get Plugin_Combobox())
   )
   > Set Initial Settings(
    self=@n55
    Project=Get Current Project()
    Selected Creature=@n56
    Developer Mode=Get Developer Mode()
    Set Plugin=@n56.Exact Match
    Animatable Creatures=Get AnimatableCreature()
    Show Project Dialog=false
    Override Anim Path Validation=false
   )
   > Set Title and Button Text(self=@n55, Title=Create New Animation, Create Button=Create, Description=Create a new Level Sequence and a Linked AnimSequence)
   n57 = GetSelectedOption(self=Get Subfolder_Combobox())
   > Branch(Condition="Not Equal, Case Insenstive (String)"!(A:String=@n57, B:String=All) -> (ReturnValue:bool))
   > Set Subfolder!(execute, self:EUW Animation Dialog Object Reference=@n55, Option:String=@n57) -> (then) {is=title K2Node_CallFunction}
   Show(execute=@n54.then_1, self=Get EUW_ContextMenu(), Show=false)
   "On Clicked (AddActorButton)"() {is=title K2Node_ComponentBoundEvent}
   > Show(self=Get EUW_ContextMenu(), Show=false)
   > n58 = Create Widget(Class=&EUW_Animation_Dialog_C)
   > AddChild(self=Get ModalLayer(), Content=@n58)
   > Set CurrentUser(CurrentUser=Get Current User(), self=@n58)
   > Bind Event to Generate Animation Paths(self=@n58, Delegate=Create Event())
   n59 = AnimatableCreature Find(
    AnimatableCreatures=Get AnimatableCreature()
    Asset=GetSelectedOption(self=Get Asset_ComboBox())
    Variant=GetSelectedOption(self=Get Plugin_Combobox())
   )
   > Set Initial Settings(
    self=@n58
    Project=Get Current Project()
    Selected Creature=@n59
    Developer Mode=Get Developer Mode()
    Set Plugin=@n59.Exact Match
    Animatable Creatures=Get AnimatableCreature()
    Show Project Dialog=false
    Override Anim Path Validation=false
   )
   > Set Title and Button Text(self=@n58, Title=Add Actor to Scene, Create Button=Add, Description=Add a new actor to your scene and create a new Linked AnimSequence)
   n60 = GetSelectedOption(self=Get Subfolder_Combobox())
   > Branch(Condition="Not Equal, Case Insenstive (String)"(A=@n60, B=All))
   > Set Subfolder(self=@n58, Option=@n60)
   "On Item Clicked (ListView)"!() -> (OutputDelegate, then, Item:Object Reference) {is=title K2Node_ComponentBoundEvent}
   > Show(self=Get EUW_ContextMenu(), Show=false)
   n61 = Get ListView()
   > Set Selected Index(
    Selected Index=GetIndexForItem!(
     self:List View Object Reference=@n61
     Item:Object Reference=Get Selected Item!(self:List View Object Reference=@n61) -> (ReturnValue:Object Reference) {is=title K2Node_CallFunction}
    ) -> (ReturnValue:int=0)
    self=Get Save Game()
   ) -> (Output_Get=0)
   > Save Settings()
   "On Clicked (FilterButton)"() {is=title K2Node_ComponentBoundEvent}
   n62 = Get EUW_ContextMenu()
   n63 = Get Name(self=@n62)
   > n64 = Branch(Condition=AND Boolean(A=IsVisible(self=@n62), B="Equal Exactly (String)"(A=@n63, B=Filter)))
   > Show(self=Get EUW_ContextMenu(), Show=false)
   Clear Context Menu(execute=@n64.else, self=Get EUW_ContextMenu())
   > n65 = Sequence!(execute) -> (then_0, then_1, then_2:Exec, then_3:Exec, then_4:Exec, then_5:Exec)
   n66 = For Each Loop(Exec=@n65.then_3, Array=Get Recent AnimSequences(self=Get Save Game()))
   > n67 = Add Icon Item(self=Get EUW_ContextMenu(), Event=Create Event(), Category=Recent Files, Name=GetBaseFilename(InPath=@n66, bRemovePath=true), Font Size=10)
   > Set String Data(String Data=@n66, self=@n67)
   Load ContextMenu Additional Project Options(execute=@n65.then_4, Context=Filter)
   Add Icon Item(execute=@n65.then_0, self=Get EUW_ContextMenu(), Event=Create Event(), Name=Reset Filters, Font Size=10)
   n68 = Get Save Game()
   Add Icon Item(
    execute=@n65.then_1
    self=Get EUW_ContextMenu()
    Event=Create Event()
    Category=Name Filters
    Name=Matches Asset Name
    Font Size=10
    Icon=Select!(
     Option 0:Texture 2D Object Reference=/AnimInEdTools/AnimationBrowser/Icons/Icon_AnimationBrowser_Unchecked
     Option 1:Texture 2D Object Reference=/AnimInEdTools/AnimationBrowser/Icons/Icon_AnimationBrowser_Checked
     Index:bool=Get Filter Match Asset Name!(self:SG Animation Browser Object Reference=@n68) -> (Filter Match Asset Name:bool)
    ) -> (ReturnValue:Texture 2D Object Reference)
   )
   n69 = GetCachedGeometry(self=Get FilterButton() -> (FilterButton:Button Object Reference))
   n70 = GetLocalSize!(Geometry:"Geometry Structure (by ref)"=@n69) -> (ReturnValue_X:double, ReturnValue_Y:double)
   Show(
    execute=@n65.then_5
    self=Get EUW_ContextMenu()
    Show=true
    Mouse Position="Local to Absolute (Coordinates)"(Geometry=@n69, LocalCoordinate_X=@n70, LocalCoordinate_Y=@n70.ReturnValue_Y)
    Name=Filter
   )
   n71 = Get Current Project()
   n72 = For Each Loop(
    Exec=@n65.then_2
    Array:Array of Gameplay Tag Structures=BreakGameplayTagContainer!(GameplayTagContainer:"Gameplay Tag Container Structure (by ref)"=Get Animatable Creature Tags!(Project:BP Animation Browser Project Object Reference=@n71, bSearchSubClasses:bool=true, __WorldContext:Object Reference) -> (Tags:Gameplay Tag Container Structure) {is=title K2Node_CallFunction}) -> (GameplayTags:Array of Gameplay Tag Structures)
   ) -> (Array Element:Gameplay Tag Structure)
   n73 = "To String (Name)"(InName=GetTagName!(GameplayTag:"Gameplay Tag Structure (by ref)"=@n72) -> (ReturnValue:Name=None))
   n74 = Get Tags!(self:SG Animation Browser Object Reference=Get Save Game()) -> (Tags:Gameplay Tag Container Structure)
   > n75 = Add Icon Item(
    self=Get EUW_ContextMenu()
    Event=Create Event()
    Category=Tag Filters
    Name=@n73
    Font Size=10
    Icon=Select(
     Option 0=&Icon_AnimationBrowser_Unchecked
     Option 1=&Icon_AnimationBrowser_Checked
     Index=HasTag!(TagContainer:"Gameplay Tag Container Structure (by ref)"=@n74, Tag:Gameplay Tag Structure=@n72, bExactMatch:bool=false) -> (ReturnValue:bool)
    )
   )
   > Set String Data(String Data=@n73, self=@n75)
   n76 = CustomEvent_2!() -> (OutputDelegate, then, Context Item:Editor Utility Widget Object Reference, Slot Widget:Editor Utility Widget Object Reference) {is=title K2Node_CustomEvent}
   > n77 = Get Selected ListView Item()
   > n78 = LoadAsset_Blocking(Asset=Get(self=@n77))
   > OpenEditorForAssets(self=Get AssetEditorSubsystem(), Assets=Make Array!("[0]"=@n78) -> (Array), OpenedMethod=Edit)
   n79 = "On Item Right Clicked (List View)"!() -> (OutputDelegate, then, Item:Object Animation Browser Object Reference, Mouse Event:Pointer Event Structure) {is=title K2Node_CustomEvent}
   > Set Selected Item(self=Get ListView(), Item=@n79.Item)
   > Clear Context Menu(self=Get EUW_ContextMenu())
   > n80 = Sequence!(execute) -> (then_0, then_1, then_2, then_3, then_4, then_5, then_6:Exec, then_7:Exec, then_8:Exec)
   n81 = Context Menu Helper - Level Sequence Has Default Map Set!(execute=@n80.then_2, self) -> (then, Level Sequence Has Default Map Set:bool) {is=title K2Node_CallFunction}
   > Branch(Condition=@n81)
   > Add Icon Item(self=Get EUW_ContextMenu(), Event=Create Event(), Category=Open, Name=Open in Level Sequence Default Map, Font Size=10)
   n82 = For Each Loop(
    Exec=@n80.then_4
    Array:Array of World Soft Object References=Get Levels!(self:BP Animation Browser Project Object Reference=Get Current Project()) -> (Levels:Array of World Soft Object References)
   ) -> (Array Element:World Soft Object Reference)
   > n83 = Branch(Condition="Equal (Integer)"(A=@n82.Array Index))
   n84 = To Soft Object Path!(SoftObjectReference=@n82) -> (ReturnValue_PathString:String)
   n85 = Add Icon Item(
    execute=@n83.else
    self=Get EUW_ContextMenu()
    Event=Create Event()
    Category=Open
    Name=Append!(A="Open in ", B=GetBaseFilename(InPath=@n84, bRemovePath=true)) -> (ReturnValue)
    Font Size=10
   )
   > Set String Data(String Data=@n84, self=@n85)
   n86 = Sequence!(execute=@n80.then_6) -> (then_0, then_1, then_2, then_3, then_4, then_5, then_6)
   n87 = CustomEvent_4!() -> (OutputDelegate, then, Context Item:Editor Utility Widget Object Reference, Slot Widget:Editor Utility Widget Object Reference) {is=title K2Node_CustomEvent}
   Add Icon Item(execute=@n86.then_1, self=Get EUW_ContextMenu(), Event=@n87, Category=Explore, Name=Browse To AnimSequence, Font Size=10)
   n88 = CopyListItemName!() -> (OutputDelegate, then, Context Item:Editor Utility Widget Object Reference, Slot Widget:Editor Utility Widget Object Reference) {is=title K2Node_CustomEvent}
   Add Icon Item(execute=@n86.then_3, self=Get EUW_ContextMenu(), Event=@n88, Category=Common, Name=Copy Name, Font Size=10)
   Add Icon Item(
    execute=@n86.then_2
    self=Get EUW_ContextMenu()
    Event=Create Event()
    Category=Montage
    Name=Create Anim Montage
    Font Size=10
    Icon=&Icon_Edit
   )
   > n89 = Get Selected ListView Item()
   > n90 = Branch(Condition=Is Empty(TargetArray:Array of Soft Object References=Get AnimMontages(self=@n89)))
   Add Icon Item(execute=@n90.else, self=Get EUW_ContextMenu(), Event=Create Event(), Category=Montage, Name=Browse To Anim Montage, Font Size=10)
   n91 = CustomEvent_6!() -> (OutputDelegate, then, Context Item:Editor Utility Widget Object Reference, Slot Widget:Editor Utility Widget Object Reference) {is=title K2Node_CustomEvent}
   Add Icon Item(
    execute=@n86.then_6
    self=Get EUW_ContextMenu()
    Event=@n91
    Category=Common
    Name=Delete
    Font Size=10
    Icon=/AnimInEdTools/ContextMenu/Icon/Icon_Delete
   )
   n92 = CustomEvent_5!() -> (OutputDelegate, then, Context Item:Editor Utility Widget Object Reference, Slot Widget:Editor Utility Widget Object Reference) {is=title K2Node_CustomEvent}
   Add Icon Item(
    execute=@n86.then_5
    self=Get EUW_ContextMenu()
    Event=@n92
    Category=Common
    Name=Duplicate and Move
    Font Size=10
    Icon=/AnimInEdTools/ContextMenu/Icon/Icon_Duplicate
   )
   n93 = CustomEvent_0!() -> (OutputDelegate, then, Context Item:Editor Utility Widget Object Reference, Slot Widget:Editor Utility Widget Object Reference) {is=title K2Node_CustomEvent}
   Add Icon Item(
    execute=@n86.then_4
    self=Get EUW_ContextMenu()
    Event=@n93
    Category=Common
    Name=Rename / Move
    Font Size=10
    Icon=/AnimInEdTools/ContextMenu/Icon/Icon_Rename
   )
   n94 = CustomEvent_3!() -> (OutputDelegate, then, Context Item:Editor Utility Widget Object Reference, Slot Widget:Editor Utility Widget Object Reference) {is=title K2Node_CustomEvent}
   Add Icon Item(execute=@n86.then_0, self=Get EUW_ContextMenu(), Event=@n94, Category=Explore, Name=Browse To LevelSequence, Font Size=10)
   Load ContextMenu Additional Project Options(execute=@n80.then_7, Context=RMB)
   Show(
    execute=@n80.then_8
    self=Get EUW_ContextMenu()
    Show=true
    Mouse Position=Get Screen Space Position!(Input:"Pointer Event Structure (by ref)"=@n79.Mouse Event) -> (ReturnValue:Vector 2D Structure) {is=title K2Node_CallFunction}
   )
   n95 = Add Icon Item(execute=@n80.then_3, self=Get EUW_ContextMenu(), Event=Create Event(), Category=Open, Name=Open in Project Default Map, Font Size=10)
   > Set String Data(String Data=Get Default Map!() -> (Map Path:String) {is=title K2Node_CallFunction}, self=@n95)
   Add Icon Item(execute=@n80.then_1, self=Get EUW_ContextMenu(), Event=Create Event(), Category=Open, Name=Open in Current Map, Font Size=10)
   Add Icon Item(
    execute=@n80.then_0
    self=Get EUW_ContextMenu()
    Event=Create Event()
    Category=General
    Name=Set Current Map as Default for Level Sequence
    Font Size=10
   )
   Add Icon Item(execute=@n80.then_5, self=Get EUW_ContextMenu(), Event=@n76, Category=Open, Name=Open AnimSequence, Font Size=10)
   n96 = Duplicate Animation!() -> (OutputDelegate, then, AnimSequence Path:String, LevelSequence Path:String, Animatable Creature:Animatable Creature Object Reference) {is=title K2Node_CustomEvent}
   > Duplicate Shot!(
    execute
    self
    Level Sequence:Soft Object Reference=Get AnimDialog Level Sequence() -> (AnimDialog Level Sequence:Soft Object Reference)
    Anim Sequence:Soft Object Reference=Get AnimDialog Anim Sequence() -> (AnimDialog Anim Sequence:Soft Object Reference)
    New Anim Sequence Path:String=@n96.AnimSequence Path
    New Level Sequence Path:String=@n96.LevelSequence Path
   ) -> (then) {is=title K2Node_CallFunction}
   > Update List View(
    Current Variation=Get Current Variation()
    Animation Type=Get Current Animation Type()
    Filter=Get Current Filter()
    Current Category=Get Current Asset()
    Use Cache=false
   )
   n97 = "Update  Level Sequence"() {is=title K2Node_CustomEvent}
   > n98 = GetFocusedLevelSequence()
   > n99 = Is Valid(InputObject=@n98)
   Branch(
    execute=@n99.Is Not Valid
    Condition=IsValidSoftObjectReference(SoftObjectReference=Get Current Level Sequence() -> (Current Level Sequence:Level Sequence Soft Object Reference))
   )
   > Set Current Level Sequence(Current Level Sequence:Level Sequence Soft Object Reference) -> (Output_Get:Level Sequence Soft Object Reference)
   > Show Currently Open Level Sequence Options!(execute, self, Show:bool=false, File Name:String) -> (then)
   n100 = Make Soft Reference!(Input:Level Sequence Object Reference=@n98) -> (Output:Level Sequence Soft Object Reference)
   n101 = Branch(
    execute=@n99.Is Valid
    Condition="Equal (SoftObjectReference)"(A:"Soft Object Reference (by ref)"=@n100, B:"Soft Object Reference (by ref)"=Get Current Level Sequence()) -> (ReturnValue:bool=~) {is=title K2Node_PromotableOperator}
   )
   n102 = Set Current Level Sequence(execute=@n101.else, Current Level Sequence=@n100)
   > Show Currently Open Level Sequence Options(
    Show=true
    File Name=GetBaseFilename(InPath=GetSoftPtrAssetName!(Asset:Soft Object Reference=@n102) -> (ReturnValue:String), bRemovePath=true)
   )
   Duplicate Dialog() {is=title K2Node_CustomEvent}
   > n103 = Get Selected ListView Item()
   n104 = Get(self=@n103)
   > Set AnimDialog Anim Sequence(AnimDialog Anim Sequence:Soft Object Reference=@n104) -> (Output_Get:Soft Object Reference)
   > Set AnimDialog Level Sequence(AnimDialog Level Sequence:Soft Object Reference=@n104.Data_LevelSequence_32_#2) -> (Output_Get:Soft Object Reference)
   > n105 = Create Widget(Class=&EUW_Animation_Dialog_C)
   > Set CurrentUser(CurrentUser=Get Current User(), self=@n105)
   > AddChild(self=Get ModalLayer(), Content=@n105)
   > Bind Event to Generate Animation Paths(self=@n105, Delegate=Create Event())
   > Set Title and Button Text(self=@n105, Title=Duplicate Animation, Create Button=Duplicate)
   n106 = "To String (SoftObjectReference)"(SoftObjectReference=Get AnimDialog Anim Sequence())
   n107 = Animatable Creature for AnimSequence!(AnimSequence Path:String=@n106, Asset Name:String=GetSelectedOption(self=Get Asset_ComboBox())) -> (AnimatableCreature:Animatable Creature Object Reference) {is=title K2Node_CallFunction}
   > Set Initial Settings(
    self=@n105
    Project=Get Current Project()
    Selected Creature=@n107
    Developer Mode=Get Developer Mode()
    Set Plugin=true
    Show Project Dialog=false
    Override Anim Path Validation=false
   )
   > Update UI from AnimSequence!(
    execute
    self:EUW Animation Dialog Object Reference=@n105
    Anim Sequence Path:String=@n106
    Animatable Creature:Animatable Creature Object Reference=@n107
   ) -> (then)
   > Show(self=Get EUW_ContextMenu(), Show=false)
   n108 = To Soft Object Path(SoftObjectReference=Get AnimDialog Anim Sequence())
   n109 = Make Array!("[0]":String=GetBaseFilename(InPath=GetPath(InPath=@n108), bRemovePath=true), "[1]":String=GetBaseFilename(InPath=@n108, bRemovePath=true)) -> (Array:Array of Strings)
   > Set Display Info!(
    execute
    self:EUW Animation Dialog Border Info Object Reference=Get HeaderWidget!(self:EUW Animation Dialog Object Reference=@n105) -> (HeaderWidget:EUW Animation Dialog Border Info Object Reference)
    Title:Text="LOCTEXT(\"Original File\")"
    Title 1:Text
    Text 1:Text="To Text (String)"(InString=Combine(InPaths=@n109))
    Title 2:Text
    Text 2:Text
    Error 1:Text
    Error 2:Text
   ) -> (then) {is=title K2Node_CallFunction}
   n110 = Edit Anim and Level Sequence!() -> (OutputDelegate, then, AnimSequence Path:String, LevelSequence Path:String, Animatable Creature:Animatable Creature Object Reference) {is=title K2Node_CustomEvent}
   > n111 = Sequence!(execute) -> (then_0, then_1, then_2)
   n112 = Get AnimDialog Level Sequence()
   > Branch(Condition=IsValidSoftObjectReference(SoftObjectReference=@n112))
   > n113 = LoadAsset_Blocking(Asset=@n112)
   > RenameLoadedAsset!(execute, SourceAsset:Object Reference=@n113, DestinationAssetPath:String=@n110.LevelSequence Path) -> (then, ReturnValue:bool)
   n114 = LoadAsset_Blocking(execute=@n111.then_1, Asset=Get AnimDialog Anim Sequence())
   > RenameLoadedAsset(SourceAsset=@n114, DestinationAssetPath=@n110.AnimSequence Path)
   SaveLoadedAssets(
    execute=@n111.then_2
    AssetsToSave=Make Array("[0]":Object Reference=@n113, "[1]":Object Reference=@n114) -> (Array:Array of Object References)
    bOnlyIfIsDirty=false
   )
   > Update List View(
    Current Variation=Get Current Variation()
    Animation Type=Get Current Animation Type()
    Filter=Get Current Filter()
    Current Category=Get Current Asset()
    Use Cache=false
   )
   Edit Animation Button On Clicked() {is=title K2Node_CustomEvent}
   > n115 = Get Selected ListView Item()
   n116 = Get(self=@n115)
   > n117 = QueryFileState!(execute, InFile:String=To Soft Object Path(SoftObjectReference=@n116), bSilent:bool=false, bUseSourceControlStateCache:bool=false) -> (
    then
    ReturnValue_Filename:String
    ReturnValue_bIsValid:bool=~
    ReturnValue_bIsUnknown:bool=~
    ReturnValue_bCanCheckIn:bool=~
    ReturnValue_bCanCheckOut:bool=~
    ReturnValue_bIsCheckedOut:bool=~
    ReturnValue_bIsCurrent:bool=~
    ReturnValue_bIsSourceControlled:bool=~
    ReturnValue_bIsAdded:bool=~
    ReturnValue_bIsDeleted:bool=~
    ReturnValue_bIsIgnored:bool=~
    ReturnValue_bCanEdit:bool=~
    ReturnValue_bCanDelete:bool=~
    ReturnValue_bIsModified:bool=~
    ReturnValue_bCanAdd:bool=~
    ReturnValue_bIsConflicted:bool=~
    ReturnValue_bCanRevert:bool=~
    ReturnValue_bIsCheckedOutOther:bool=~
    ReturnValue_CheckedOutOther:String
    ReturnValue_bIsCheckedOutInOtherBranch:bool=~
    ReturnValue_bIsModifiedInOtherBranch:bool=~
    ReturnValue_PreviousUser:String
   )
   > n118 = Branch(Condition=@n117.ReturnValue_bIsCheckedOutOther)
   Show Message Dialog(
    execute=[@n128.then, @n118.then]
    Title=MakeLiteralText(Value="LOCTEXT(\"Cannot Edit File\")")
    Message=MakeLiteralText(Value="LOCTEXT(\"The selected file is currently checked out by someone else.\")")
    MessageType=Ok
    DefaultValue=No
    MessageCategory=Warning
   )
   n119 = Resolve Soft Reference!(Input:Soft Object Reference=@n116.Data_LevelSequence_32_#2) -> (Output:Object Reference)
   n120 = Is Valid(exec=@n118.else, InputObject=@n119)
   n121 = Create Widget(execute=[@n128.else, @n120.Is Not Valid], Class=&EUW_Animation_Dialog_C)
   > Set CurrentUser(CurrentUser=Get Current User(), self=@n121)
   > AddChild(self=Get ModalLayer(), Content=@n121)
   > Set Title and Button Text(self=@n121, Title=Rename Animation, Create Button=Edit)
   > Bind Event to Generate Animation Paths(self=@n121, Delegate=Create Event())
   n122 = Get(self=@n115)
   n123 = "To String (SoftObjectReference)"(SoftObjectReference=@n122)
   n124 = Animatable Creature for AnimSequence(AnimSequence Path=@n123, Asset Name=GetSelectedOption(self=Get Asset_ComboBox()))
   > Set Initial Settings(
    self=@n121
    Project=Get Current Project()
    Selected Creature=@n124
    Developer Mode=Get Developer Mode()
    Set Plugin=true
    Show Project Dialog=false
    Override Anim Path Validation=false
   )
   > Update UI from AnimSequence(self=@n121, Anim Sequence Path=@n123, Animatable Creature=@n124)
   > Show(self=Get EUW_ContextMenu(), Show=false)
   n125 = To Soft Object Path(SoftObjectReference=@n122)
   n126 = Make Array("[0]":String=GetBaseFilename(InPath=GetPath(InPath=@n125), bRemovePath=true), "[1]":String=GetBaseFilename(InPath=@n125, bRemovePath=true)) -> (Array:Array of Strings)
   > Set Display Info(self=Get HeaderWidget(self=@n121), Title="LOCTEXT(\"Original File\")", Text 1="To Text (String)"(InString=Combine(InPaths=@n126)))
   > Set AnimDialog Anim Sequence(AnimDialog Anim Sequence=@n122)
   > Set AnimDialog Level Sequence(AnimDialog Level Sequence=@n122.Data_LevelSequence_32_#2)
   n127 = QueryFileState(
    execute=@n120.Is Valid
    InFile=GetBaseFilename(InPath=Get Object Path String(Object=@n119), bRemovePath=false)
    bSilent=false
    bUseSourceControlStateCache=false
   ) -> (
    ReturnValue_bIsValid=~
    ReturnValue_bIsUnknown=~
    ReturnValue_bCanCheckIn=~
    ReturnValue_bCanCheckOut=~
    ReturnValue_bIsCheckedOut=~
    ReturnValue_bIsCurrent=~
    ReturnValue_bIsSourceControlled=~
    ReturnValue_bIsAdded=~
    ReturnValue_bIsDeleted=~
    ReturnValue_bIsIgnored=~
    ReturnValue_bCanEdit=~
    ReturnValue_bCanDelete=~
    ReturnValue_bIsModified=~
    ReturnValue_bCanAdd=~
    ReturnValue_bIsConflicted=~
    ReturnValue_bCanRevert=~
    ReturnValue_bIsCheckedOutOther=~
    ReturnValue_bIsCheckedOutInOtherBranch=~
    ReturnValue_bIsModifiedInOtherBranch=~
   )
   > n128 = Branch(Condition=@n127.ReturnValue_bIsCheckedOutOther)
   Save Widget State() {is=title K2Node_CustomEvent}
   > n129 = Is Valid(InputObject=Get Save Game())
   n130 = CreateSaveGameObject!(
    execute=@n129.Is Not Valid
    SaveGameClass:Save Game Class Reference=&/SG_AnimationBrowser.SG_AnimationBrowser_C
   ) -> (then, ReturnValue:SG Animation Browser Object Reference)
   > n131 = Set Save Game(Save Game:SG Animation Browser Object Reference=@n130) -> (Output_Get:SG Animation Browser Object Reference)
   n132 = Get Save Game()
   Set Variant(execute=[@n129.Is Valid, @n131.then], Variant=Get Current Variation(), self=@n132)
   > Set Asset(Asset=Get Current Asset(), self=@n132)
   > Set Subfolder(Subfolder=GetSelectedOption(self=Get Subfolder_Combobox()), self=@n132) {is=title K2Node_VariableSet}
   > Set Search(Search=Get Text(self:Text Box Object Reference=Get SearchTextBox()) -> (Text:Text), self=@n132)
   n133 = Get ListView()
   n134 = Get Selected Item(self=@n133)
   n135 = Get Save Game()
   > Set Selected Index(Selected Index=GetIndexForItem(self=@n133, Item=@n134) -> (ReturnValue=0), self=@n135) -> (Output_Get=0)
   > n136 = Sequence()
   > n137 = Cast To Object_AnimationBrowser(Object=@n134)
   > Set Selected Path(Selected Path=To Soft Object Path(SoftObjectReference=Get(self=@n137)), self=@n135)
   Set Project!(execute=@n136.then_1, Project:BP Animation Browser Project Soft Object Reference, self:SG Animation Browser Object Reference=@n135) -> (then, Output_Get:BP Animation Browser Project Soft Object Reference)
   > Set Developer Mode!(execute, Developer Mode:bool=Get Developer Mode(), self:SG Animation Browser Object Reference=@n135) -> (then, Output_Get:bool)
   AsyncSaveGameToSlot!(
    execute:Exec=@n136.then_2
    WorldContextObject:Object Reference
    SaveGameObject:Save Game Object Reference=@n135
    SlotName:String=AnimationBrowser
    UserIndex:int=0
   ) -> (then:Exec, Completed:Exec, SaveGame:Save Game Object Reference, bSuccess:bool=~) {is=title K2Node_AsyncAction}
   n138 = Update Character Combo Box() {is=title K2Node_CustomEvent}
   Refresh Assets() {is=title K2Node_CustomEvent}
   > n139 = Branch(Condition=IsLoadingAssets!(self:Asset Registry Interface=GetAssetRegistry()) -> (ReturnValue:bool))
   n140 = Get Current Project()
   n141 = Set AnimatableCreature(
    execute=@n139.else
    AnimatableCreature:Array of Animatable Creature Object References=Get Animatable Creatures!(Project:BP Animation Browser Project Object Reference=@n140, bSearchSubClasses:bool=false, __WorldContext:Object Reference) -> (Animatable Creatures:Array of Animatable Creature Object References) {is=title K2Node_CallFunction}
   ) -> (Output_Get:Array of Animatable Creature Object References)
   > Append Array!(
    execute
    TargetArray:Array of Animatable Creature Object References=@n141
    SourceArray:Array of Animatable Creature Object References=Get SG Animatable Creatures!(Project:BP Animation Browser Project Object Reference=@n140, __WorldContext:Object Reference) -> (Animatable Creatures:Array of Animatable Creature Shot Grid Object References) {is=title K2Node_CallFunction}
   ) -> (then)
   n142 = Backup Level Sequence Event() {is=title K2Node_CustomEvent}
   > n143 = GetFocusedLevelSequence()
   > Is Valid(InputObject=@n143)
   > Branch(Condition=FileExists!(InPath:String=Get Object Path String(Object=@n143)) -> (ReturnValue:bool))
   > Backup Level Sequence(Level Sequence=@n143)
   Context Menu Open Level Sequence!() -> (OutputDelegate, then, Context Item:Editor Utility Widget Object Reference, Slot Widget:Editor Utility Widget Object Reference) {is=title K2Node_CustomEvent}
   > n144 = Get Selected ListView Item()
   > n145 = Async Load Asset(Asset=Break Struct Animation Brower Item!(Struct_AnimationBrower_Item=Get Data!(self:Object Animation Browser Object Reference=@n144) -> (Data:Struct Animation Brower Item Structure)) -> (
     AnimSequence_31_#1
     LevelSequence_32_#2:Soft Object Reference
     CurrentUser_26_#3:String
    ))
   n146 = Cast To AnimSequence(execute=@n145.Completed, Object=@n145)
   > PrintString(InString=GetObjectName(Object=@n146))
   > Open Item(Item=@n144)
   Event Construct()
   > Set Current Animation Type(Current Animation Type:String=Get Animation Type Default Item()) -> (Output_Get:String)
   > Display Modal Dialog(Class=&/EUW_RegistryLoadingDialog.EUW_RegistryLoadingDialog_C) {is=title K2Node_CallFunction}
   > Execute Python Script!(execute, PythonScript="import AnimationBrowser\r\n\r\nfrom importlib import reload\r\nreload(AnimationBrowser)") -> (then, ReturnValue)
   > SetScalarParameterValue!(
    execute
    WorldContextObject
    Collection:Material Parameter Collection Object Reference
    ParameterName:Name=FigureBodySSS
    ParameterValue:float=0
   ) -> (then) {is=title K2Node_CallMaterialParameterCollectionFunction}
   Event Tick()
   > n147 = Branch(Condition=Get Is Setup() -> (Is Setup:bool))
   n148 = Branch(execute=@n147.else, Condition=IsLoadingAssets(self=GetAssetRegistry()))
   Refresh Assets(execute=@n148.else) {is=title K2Node_CallFunction}
   > n149 = Sequence!(execute) -> (then_0, then_1, then_2, then_3:Exec, then_4:Exec, then_5:Exec)
   Set Is Setup(execute=@n149.then_1, Is Setup:bool=true) -> (Output_Get:bool)
   > Load Save Game From Slot()
   n150 = Is Valid(exec=@n149.then_2, InputObject=Get Current Project())
   n151 = Display Modal Dialog(execute=@n150.Is Not Valid, Class=&/EUW_SelectProjectDialog.EUW_SelectProjectDialog_C)
   > n152 = Cast To EUW_SelectProjectDialog!(execute, Object=@n151.New Widget) -> (then, CastFailed, AsEUW Select Project Dialog:EUW Select Project Dialog Object Reference) {is=title K2Node_DynamicCast}
   > Set Projects!(
    execute
    self:EUW Select Project Dialog Object Reference=@n152
    Projects:Array of BP Animation Browser Project Object References=Get Projects()
   ) -> (then) {is=title K2Node_CallFunction}
   > Bind Event to Project Selected(self:EUW Select Project Dialog Object Reference=@n152, Delegate=Create Event()) {is=title K2Node_AddDelegate}
   Refresh Assets(execute=@n150.Is Valid)
   Update Subfolder List(execute=@n149.then_5)
   Set Timer by Event(execute=@n149.then_3, Delegate=@n142, Time=15, bLooping=true)
   Set Timer by Event(execute=@n149.then_4, Delegate=@n97, Time=2, bLooping=true)
   Branch(execute=@n149.then_0, Condition=IsEmpty(InString=Get Current User()))
   > n153 = Execute Python Script!(execute, PythonScript="import os\r\nuser = os.getlogin()") -> (then, ReturnValue, user:String) {outputs [user]}
   > Set Current User(Current User:String=@n153.user) -> (Output_Get:String)
   n154 = Branch(Condition=Get isShotgridEnabled())
   n155 = Get Selected ListView Item(execute=@n94.then)
   n156 = Get(self=@n155)
   SyncBrowserToObjects(
    execute=@n178.then
    AssetPaths=Make Array!("[0]"=To Soft Object Path(SoftObjectReference=@n156.Data_LevelSequence_32_#2)) -> (Array)
   )
   GetLocalSize(Geometry=@n8)
   n157 = Get Selected ListView Item(execute=@n91.then)
   > Delete Item!(execute, self, Data:Object Animation Browser Object Reference=@n157) -> (then) {is=title K2Node_CallFunction}
   n158 = GetFocusedLevelSequence(execute=@n165.then)
   n159 = Get Object Path String(Object=@n158)
   n160 = Get AutoSave Dir()
   n161 = Cast To EUW_Restore_ContextMenu_Item(execute=@n22.then, Object=@n22.Slot Widget)
   n162 = Get File Path!(self:EUW Restore Context Menu Item Object Reference=@n161) -> (File Path:String)
   n163 = ConvertAssetPathToFilePath!(
    execute=@n179.then
    InAssetPath:String=RightChop!(SourceString:String=MakePathRelativeTo!(InPath:String=@n162, InRelativeTo:String=@n160) -> (OutPath:String, ReturnValue:bool), Count:int=9) -> (ReturnValue:String)
    InExtension:String
   ) -> (then, OutPath:String, ReturnValue:bool)
   n164 = ConvertToAbsolutePath!(Filename:String=@n163) -> (ReturnValue:String)
   n165 = ConvertFilePathToAssetPath!(execute=@n176.then, InPath:String=@n164) -> (then, OutPath:String, ReturnValue:bool)
   n166 = LoadAsset(execute=@n174.then, AssetPath=Append!(A=@n165, B=., C:String=GetBaseFilename(InPath=@n159, bRemovePath=true)) -> (ReturnValue))
   > n167 = Cast To LevelSequence(Object=@n166)
   > Move Binding To Level Sequence!(
    execute
    self
    Source Level Sequence:Movie Scene Sequence Object Reference=@n167
    Destination Level Sequence:Movie Scene Sequence Object Reference=@n158
   ) -> (then)
   n168 = Get Selected ListView Item(execute=@n87.then)
   n169 = Get(self=@n168)
   SyncBrowserToObjects(execute=@n177.then, AssetPaths=Make Array("[0]"=To Soft Object Path(SoftObjectReference=@n169)))
   Edit Animation Button On Clicked(execute=@n93.then) {is=title K2Node_CallFunction}
   Duplicate Dialog(execute=@n92.then) {is=title K2Node_CallFunction}
   Make Soft Reference(Input:BP Animation Browser Project Object Reference=Get Current Project()) -> (Output:BP Animation Browser Project Soft Object Reference)
   Has Set Default Map Event!() -> (OutputDelegate, then, Context Item:Editor Utility Widget Object Reference, Slot Widget:Editor Utility Widget Object Reference) {is=title K2Node_CustomEvent}
   n170 = Get TextBlock!(self:Animation Browser List Item Object Reference=Get Current List Item()) -> (TextBlock:Text Object Reference)
   Execute Python Script!(
    execute=@n88.then
    PythonScript="import subprocess\r\nimport sys\r\n\r\nif sys.platform == \"win32\":\r\n    subprocess.run(\"clip\", input=text.encode(), check=True)\r\nelif sys.platform == \"darwin\":\r\n    subprocess.run(\"pbcopy\", input=text.encode(), check=True)\r\nelse:\r\n    for cmd in [[\"xclip\", \"-selection\", \"clipboard\"], [\"xsel\", \"--clipboard\", \"--input\"], [\"wl-copy\"]]:\r\n        try:\r\n            subprocess.run(cmd, input=text.encode(), check=True)\r\n            break\r\n        except FileNotFoundError:\r\n            continue"
    text:String="To String (Text)"(InText=Get Text(self:Text Object Reference=@n170))
   ) -> (then, ReturnValue) {inputs [text]}
   n171 = IsValidSoftObjectReference(SoftObjectReference=@n156.Data_LevelSequence_32_#2)
   n172 = FileExists(InPath=@n162)
   n173 = GetAssetRegistry()
   n174 = ScanFilesSynchronous!(execute=@n158.then, self:Asset Registry Interface=@n173, InFilePaths:Array of Strings=Make Array("[0]"=@n165), bForceRescan:bool=true) -> (then)
   n175 = IsValidSoftObjectReference(SoftObjectReference=@n169)
   n176 = Execute Python Script!(
    execute=@n163.then
    PythonScript="import os\r\nimport shutil\r\nfrom pathlib import Path\r\n\r\nprint(src, dst)\r\nif not os.path.exists(dst):\r\n    ## mkdir\r\n    Path(dst).parent.mkdir(parents=True, exist_ok=True)\r\n\r\n    ## Copy File\r\n    shutil.copyfile(src, dst)"
    src:String=@n162
    dst:String=@n164
   ) -> (then, ReturnValue) {inputs [src,dst]}
   n177 = Branch(execute=@n168.then, Condition=@n175)
   n178 = Branch(execute=@n155.then, Condition=@n171)
   n179 = Branch(execute=@n161.then, Condition=@n172)
  ]}
 Filter Menu - Open Recent {nodes [
   n1 = Filter Menu - Open Recent!() -> (then, Context Item:Editor Utility Widget Object Reference, Slot Widget:Editor Utility Widget Object Reference) {is=title K2Node_FunctionEntry}
   > n2 = Cast To EUW_ContextMenu_Icon(Object=@n1.Slot Widget)
   n3 = Get String Data(self=@n2)
   > n4 = LoadAsset(AssetPath=@n3)
   > n5 = Cast To LevelSequence(Object=@n4)
   > Open Sequence in Level!(execute, self, LevelSequence:Level Sequence Object Reference=@n5, Level Path:String) -> (then)
   > Add LevelSequence to Recents!(execute, self, LevelSequence:"String (by ref)"=@n3) -> (then) {is=title K2Node_CallFunction}
  ]}
 Filter Menu - Reset Filters {nodes [
   Filter Menu - Reset Filters!() -> (then, Context Item:Editor Utility Widget Object Reference, Slot Widget:Editor Utility Widget Object Reference) {is=title K2Node_FunctionEntry}
   > Set Tags!(execute, Tags:Gameplay Tag Container Structure, self:SG Animation Browser Object Reference=Get Save Game()) -> (then, Output_Get:Gameplay Tag Container Structure)
   > Set Filter Match Asset Name!(execute, Filter Match Asset Name:bool=false, self:SG Animation Browser Object Reference=Get Save Game()) -> (then, Output_Get:bool)
   > SaveGameToSlot(SaveGameObject=Get Save Game(), SlotName=AnimationBrowser, UserIndex=0)
   > Update List View From UI()
  ]}
 Filter Menu - Toggle Match Asset Name {nodes [
   Filter Menu - Toggle Match Asset Name!() -> (then, Context Item:Editor Utility Widget Object Reference, Slot Widget:Editor Utility Widget Object Reference) {is=title K2Node_FunctionEntry}
   n1 = Get Save Game()
   > Set Filter Match Asset Name(Filter Match Asset Name=NOT Boolean(A=Get Filter Match Asset Name(self=@n1)), self=Get Save Game())
   > SaveGameToSlot(SaveGameObject=Get Save Game(), SlotName=AnimationBrowser, UserIndex=0)
   > Update List View From UI()
  ]}
 Filter Menu - Toggle Tag {nodes [
   n1 = Filter Menu - Toggle Tag!() -> (then, Context Item:Editor Utility Widget Object Reference, Slot Widget:Editor Utility Widget Object Reference) {is=title K2Node_FunctionEntry}
   > n2 = Cast To EUW_ContextMenu_Icon(Object=@n1.Slot Widget)
   > n3 = RequestGameplayTag!(execute, TagName:String=Get String Data(self=@n2), bErrorIfNotFound:bool=false) -> (then, OutTag:Gameplay Tag Structure, ReturnValue:bool)
   > AddGameplayTag!(execute, TagContainer:"Gameplay Tag Container Structure (by ref)"=Get Tags(self=Get Save Game()), Tag:Gameplay Tag Structure=@n3) -> (then)
   > SaveGameToSlot(SaveGameObject=Get Save Game(), SlotName=AnimationBrowser, UserIndex=0)
  ]}
 Find Animatable Creature Using Name Semantics {nodes [
   Find Animatable Creature Using Name Semantics!() -> (then, Asset Reference:"Soft Object Reference (by ref)") {is=title K2Node_FunctionEntry}
   > n1 = Sequence!(execute) -> (then_0, then_1)
   n2 = For Each Loop(Exec=@n1.then_1, Array:Array of Animatable Creature Object References=Get AnimatableCreature()) -> (Array Element:Animatable Creature Object Reference)
   Return Node!(
    execute=@n2.Completed
    Best Match AC File:Animatable Creature Object Reference=Get Found AC File() -> (Found AC File:Animatable Creature Object Reference)
   )
   n3 = Get Asset!(self:Animatable Creature Object Reference=@n2) -> (Asset)
   > Branch(Condition="Equal Exactly (String)"(A=Get Current Asset(), B=@n3))
   n4 = Get Current Variation()
   n5 = Get Variant(self=@n2)
   > Branch(Condition=OR Boolean(A="Equal Exactly (String)"(A=@n4, B=All), B="Equal Exactly (String)"(A=@n4, B=@n5)))
   > n6 = Execute Python Script!(
    execute
    PythonScript="import re\r\n\r\nif animation_path.find(\"{\") >= 0:\r\n    animation_path = animation_path[0:animation_path.find(\"{\")]\r\n\r\n## Match the path to determine the confidence that we are in the right place. We expect to match\r\n## at index position 0 or not at all. However, if we get a positive integer, we place a penalty,\r\n## effectively lowering the confidence value.\r\npath_confidence = -1\r\npos = anim_sequence_name_path.find(animation_path)\r\nif pos >= 0:\r\n    path_confidence = max(len(animation_path) - pos, 0)\r\n\r\n## Append a slash to enforce strict word matching\r\nanim_sequence_name_path = anim_sequence_name_path + \"/\"\r\n\r\nfound_any_match = False\r\nresult_score = 0\r\nfor word in search_words:\r\n    cleanupstring = re.sub(r'[^a-zA-Z0-9]', '/', anim_sequence_name_path)\r\n    score = cleanupstring.find(word + \"/\")\r\n    if score == -1:\r\n        score = len(cleanupstring)\r\n    else:\r\n        found_any_match = True\r\n    result_score += score\r\n\r\nprint(\"{} || {}|{}> {}\".format(found_any_match, path_confidence, score, anim_sequence_name_path))"
    SearchWords:Array of Strings=Make Array!("[0]"=@n3, "[1]":String=@n5) -> (Array)
    AnimSequenceNamePath:String="To String (SoftObjectReference)"(SoftObjectReference=Get Asset Reference() -> (Asset Reference:"Soft Object Reference (by ref)"))
    AnimationPath:String=Get Animation Path(self=@n2)
   ) -> (then, ReturnValue, PathConfidence:int, ResultScore:int, FoundAnyMatch:bool=~) {inputs [SearchWords, AnimSequenceNamePath, AnimationPath], outputs [PathConfidence, ResultScore, FoundAnyMatch]}
   > Branch(Condition=@n6.FoundAnyMatch)
   > n7 = Branch(Condition="integer > integer"(A=@n6.PathConfidence, B=Get PathConfidence() -> (PathConfidence:int=0)))
   n8 = Sequence!(execute=[@n9.then, @n7.then]) -> (then_0, then_1, then_2:Exec)
   > Set Found AC File(Found AC File:Animatable Creature Object Reference=@n2) -> (Output_Get:Animatable Creature Object Reference)
   Set PathConfidence(execute=@n8.then_1, PathConfidence:int=@n6.PathConfidence) -> (Output_Get:int=0)
   Set StringMatchIndex(execute=@n8.then_2, StringMatchIndex:int=@n6.ResultScore) -> (Output_Get:int=0)
   n9 = Branch(execute=@n7.else, Condition="integer < integer"(A=@n6.ResultScore, B=Get StringMatchIndex() -> (StringMatchIndex:int=0)))
  ]}
 Generate AnimSequence Path {nodes [
   n1 = Generate AnimSequence Path!() -> (then, Animation Type:String, Character:String, Category:String, Name:String, Suffix:String, Character Category:String) {is=title K2Node_FunctionEntry}
   > n2 = Get Animation Directory Path!(
    execute
    self
    Character:"String (by ref)"=Get Character() -> (Character:String)
    Category:String=Get Character Category() -> (Character Category:String)
   ) -> (then, Animation Directory:String) {is=title K2Node_CallFunction}
   n3 = Get Suffix() -> (Suffix:String)
   n4 = IsEmpty(InString=Get Name!() -> (Name))
   n5 = Get Current Character Name!(Category:String=@n1.Character Category, Character:String=@n1.Character) -> (NewParam:String) {is=title K2Node_CallFunction}
   n6 = Select(
    Option 0:String=Append!(A=@n5, B=_, C=Get Category() -> (Category:String), D:String=_, E:String=Get Name()) -> (ReturnValue)
    Option 1:String=Append!(A=@n5, B=_, C=Get Name()) -> (ReturnValue)
    Index=IsEmpty(InString=Get Category())
   ) -> (ReturnValue:String)
   n7 = Append!(A=@n6, B=Get Suffix()) -> (ReturnValue)
   n8 = Append!(A=@n6, B=_, C:String=Get Suffix()) -> (ReturnValue)
   n9 = MakeValidFileName!(
    InString:String=Select(Option 0=Select(Option 0=@n8, Option 1=@n7, Index=@n4), Option 1=@n6, Index=IsEmpty(InString=@n3))
    InReplacementChar:String
   ) -> (ReturnValue:String)
   n10 = Get Animation Type() -> (Animation Type:String)
   > Return Node!(execute, Path:String=Combine(InPaths=Make Array!("[0]"=@n2, "[1]"=@n10, "[2]":String=@n9) -> (Array)))
  ]}
 Generate LevelSequence Path {nodes [
   n1 = Generate LevelSequence Path!() -> (then, Animation Type:String, Character:String, Category:String, Name:String, Suffix:String, Character Category:String) {is=title K2Node_FunctionEntry}
   > n2 = Get Level Sequence Directory Path!(execute, self, Character:"String (by ref)"=@n1.Character, Category:String=@n1.Character Category) -> (then, Animation Directory:String) {is=title K2Node_CallFunction}
   > n3 = Get Animation Directory Path(Character=@n1.Character, Category=@n1.Character Category)
   n4 = GetCleanFilename(InPath=Generate AnimSequence Path!(
     Animation Type:String=@n1
     Character:String=@n1.Character
     Category:String=@n1.Category
     Name:String=@n1.Name
     Suffix:String=@n1.Suffix
     Character Category:String=@n1.Character Category
    ) -> (Path:String) {is=title K2Node_CallFunction})
   n5 = Select(Option 1=Source, Index="Equal, Case Insensitive (String)"(A=@n2, B=@n3))
   > Return Node(Path=Combine(InPaths=Make Array!("[0]"=@n2, "[1]"=@n1, "[2]"=@n5, "[3]":String=@n4) -> (Array)))
  ]}
 Get AnimMontage Path From AnimSequence {nodes [
   n1 = Get AnimMontage Path From AnimSequence!() -> (then, AnimSequence Path:String, InBinding:"Movie Scene Binding Proxy Structure (by ref)") {is=title K2Node_FunctionEntry}
   > n2 = Find Animatable Creature Using Name Semantics!(
    execute
    self
    Asset Reference:"Soft Object Reference (by ref)"=To Soft Object Reference!(SoftObjectPath:"Soft Object Path Structure (by ref)"=MakeSoftObjectPath!(PathString:String=@n1) -> (ReturnValue:Soft Object Path Structure)) -> (ReturnValue)
   ) -> (then, Best Match AC File:Animatable Creature Object Reference) {is=title K2Node_CallFunction}
   > n3 = Set AC File(AC File:Animatable Creature Object Reference=@n2) -> (Output_Get:Animatable Creature Object Reference)
   > n4 = Is Valid(InputObject=@n3)
   Return Node!(execute=@n4.Is Not Valid, AnimMontage Path:String, Is Valid:bool=false)
   > n5 = Sequence!(execute) -> (then_0, then_1)
   n6 = Get AnimMontage Path!(self:Animatable Creature Object Reference=@n3) -> (AnimMontage Path:String)
   > n7 = Branch(Condition=OR Boolean(A=IsEmpty(InString=@n6), B="Equal Exactly (String)"(A=@n6, B=())))
   > Set Montage Path(Montage Path=GetPath(InPath=Get AnimSequence Path()))
   n8 = Get AC File() -> (AC File:Animatable Creature Object Reference)
   n9 = Get Asset(self=@n8)
   n10 = Get Variant(self=@n8)
   Set Montage Path(
    execute=@n7.else
    Montage Path=Get Path!(
     self:EU Animation Naming Convention Object Reference=Get Naming Convention!(self:Animatable Creature Object Reference=@n8) -> (ReturnValue:EU Animation Naming Convention Object Reference) {is=title K2Node_CallFunction}
     Token Map:Map of Strings to Strings=Get Path Parts Map!(
      Asset:String=@n9
      Animation Set:String
      Animation Name:String
      Descriptor:String
      Direction:String
      Subfolder:String
      Variation:String
      Plugin:String=@n10
      __WorldContext:Object Reference
     ) -> (Output Path Parts Map:Map of Strings to Strings) {is=title K2Node_CallFunction}
     Path Pattern:String=Get AnimMontage Path(self=@n8)
     Remove Spaces On Name:bool=Get RemoveSpacesOnName!(self:Animatable Creature Object Reference=@n8) -> (RemoveSpacesOnName:bool)
    ) -> (File Path:String, Folder Path:String, File Name:String) {is=title K2Node_CallFunction}.Folder Path
   )
   Return Node(execute=@n5.then_1, AnimMontage Path=Get Montage Path(), Is Valid=true)
   Reroute Node(InputPin:Animatable Creature Object Reference=Find Animatable Creature from AnimSequence!(
     AnimSequence Path:String=@n1
     Blueprint Actor:Object Class Reference=GetPossessedObjectClass!(execute, InBinding:"Movie Scene Binding Proxy Structure (by ref)") -> (then, ReturnValue:Object Class Reference)
     Project:BP Animation Browser Project Object Reference=Get Current Project()
     __WorldContext:Object Reference
    ) -> (Animatable Creatures:Animatable Creature Object Reference)) -> (OutputPin:Animatable Creature Object Reference)
  ]}
 Get Animation Directories for Asset {nodes [
   Get Animation Directories for Asset!() -> (then, Asset Name:String, Variation:String, Animation Type:String) {is=title K2Node_FunctionEntry}
   > Clear(TargetArray:Array of Names=Get Local Directories() -> (Local Directories:Array of Names))
   > n1 = For Each Loop(Array=Get AnimatableCreature())
   Return Node!(execute=@n1.Completed, Animation Directories:Array of Names=Get Local Directories())
   n2 = Get Asset Name()
   n3 = Get Variant(self=@n1)
   n4 = Get Asset(self=@n1)
   n5 = "Equal Exactly (String)"(A=Get Asset(self=@n1), B=Get Asset Name())
   n6 = OR Boolean(
    A="Equal Exactly (String)"(A=Get Variant(self=@n1), B=Get Variation() -> (Variation:String))
    B="Equal Exactly (String)"(A=Get Variation(), B=All)
   )
   > Branch(Condition=Select(Option 0:bool=AND Boolean(A=@n6, B=@n5), Option 1:bool="Equal Exactly (String)"(A=@n3, B=@n2), Index=IsEmpty(InString=@n4)) -> (ReturnValue:bool))
   n7 = Get Populate Subfolders From Disk!(self:Animatable Creature Object Reference=@n1) -> (Populate Subfolders From Disk:bool)
   > n8 = Branch(Condition=@n7)
   > n9 = Get Subfolders From Disk!(execute, self, Animatable Creature:Animatable Creature Object Reference=@n1) -> (then, Subfolders:Array of Strings) {is=title K2Node_CallFunction}
   n10 = Get Animation Type Default Item()
   n11 = Get Animation Type()
   n12 = Get Subfolders!(self:Animatable Creature Object Reference=@n1) -> (Subfolders)
   n13 = For Each Loop(
    Exec=[@n9.then, @n8.else]
    Array:Array of Strings=Select(
     Option 0:Array of Strings=Make Array!("[0]"=@n11) -> (Array)
     Option 1:Array of Strings=Select(Option 0=@n12, Option 1=@n9, Index=@n7)
     Index="Equal Exactly (String)"(A=@n11, B=@n10)
    ) -> (ReturnValue:Array of Strings)
   ) -> (Array Element:String)
   n14 = Get Path(
    self=Get Naming Convention(self=@n1)
    Token Map=Get Path Parts Map(Asset=Get Asset(self=@n1), Subfolder=@n13, Plugin=Get Variant(self=@n1))
    Path Pattern=Get Animation Path(self=@n1)
    Remove Spaces On Name=Get RemoveSpacesOnName(self=@n1)
   )
   > n15 = DoesDirectoryExist!(execute, DirectoryPath:String=@n14.Folder Path) -> (then, ReturnValue:bool)
   > Branch(Condition=@n15)
   > Add(TargetArray:Array of Names=Get Local Directories(), NewItem:"Name (by ref)"=String To Name(InString=@n14.Folder Path))
  ]}
 Get Animation Directory Path {nodes [
   n1 = Get Animation Directory Path!() -> (then, Character:"String (by ref)", Category:String) {is=title K2Node_FunctionEntry}
   n2 = Find Animatable Creature!(
    Asset:String=@n1.Category
    Variation:String=@n1
    Project:BP Animation Browser Project Object Reference=Get Current Project()
    __WorldContext:Object Reference
   ) -> (Animatable Creatures:Animatable Creature Object Reference)
   > n3 = Is Valid(InputObject=@n2)
   Return Node!(execute=@n3.Is Not Valid, Animation Directory:String)
   > Return Node(Animation Directory=Get Animation Path(self=@n2))
  ]}
 Get AutoSave Dir {nodes [
   Get AutoSave Dir() {is=title K2Node_FunctionEntry}
   n1 = Combine(InPaths=Make Array!("[0]"=GetGameFeaturesDirAbsolutePath!() -> (ReturnValue:String), "[1]":String=ProjectSavedDir!() -> (ReturnValue:String)) -> (Array))
   n2 = Combine(InPaths=Make Array("[0]"=CollapseRelativeDirectories!(InPath:String=@n1) -> (OutPath:String, ReturnValue:bool), "[1]"=Autosaves))
   > Return Node!(execute, AutoSave Dir:String=@n2, Exists:bool=DirectoryExists!(InPath:String=@n2) -> (ReturnValue:bool))
  ]}
 Get Base Path {nodes [
   n1 = Get Base Path!() -> (then, Path:String) {is=title K2Node_FunctionEntry}
   n2 = GetBaseFilename(InPath=@n1, bRemovePath=true)
   n3 = GetPath(InPath=@n1)
   > Return Node!(execute, Output:String=Combine(InPaths=Make Array("[0]"=@n3, "[1]"=@n2)))
  ]}
 Get Current Character Name {nodes [
   n1 = Get Current Character Name!() -> (then, Category:String, Character:String) {is=title K2Node_FunctionEntry}
   n2 = Get Character()
   n3 = "Equal, Case Insensitive (String)"(A=@n1.Character, B=Shared)
   n4 = Append(A=@n1, B=_, C=@n1.Character)
   > Return Node!(
    execute
    NewParam:String=Select(Option 0:String=Select(Option 0=@n4, Option 1=@n1, Index=@n3), Option 1:String=Get Category(), Index=IsEmpty(InString=@n2)) -> (ReturnValue:String)
   )
  ]}
 Get Default Map {nodes [
   Get Default Map() {is=title K2Node_FunctionEntry}
   n1 = Get Levels(self=Get Current Project())
   n2 = "Get (a copy)"(Array:Array of World Soft Object References=@n1) -> (Output:World Soft Object Reference)
   n3 = Make Array("[0]"=/AnimInEdTools/AnimationBrowser/, "[1]"=Assets/AnimationBrowser_DefaultLevel)
   n4 = IsValidSoftObjectReference(SoftObjectReference=@n2)
   n5 = "integer >= integer"(A=Length(TargetArray:Array of World Soft Object References=@n1), B=1)
   > Return Node!(
    execute
    Map Path:String=SelectString(A=To Soft Object Path(SoftObjectReference=@n2), B=Combine(InPaths=@n3), bPickA=AND Boolean(A=@n5, B=@n4))
   )
  ]}
 Get Default Map for Level Sequence {nodes [
   n1 = Get Default Map for Level Sequence!() -> (then, Level Sequence:Level Sequence Object Reference) {is=title K2Node_FunctionEntry}
   > n2 = GetMetadataTag!(execute, Object:Object Reference=@n1, Tag:Name=Get Default Map Tag() -> (Default Map Tag:Name=None)) -> (then, ReturnValue:String)
   > Return Node!(execute, Default Map:String=@n2)
  ]}
 Get Level Sequence Backups {nodes [
   n1 = Get Level Sequence Backups!() -> (then, Level Sequence:Level Sequence Object Reference) {is=title K2Node_FunctionEntry}
   n2 = Get AutoSave Dir() {is=title K2Node_CallFunction}
   > Branch(Condition=@n2.Exists)
   n3 = Get Object Path String(Object=@n1)
   n4 = GetPath(InPath=Combine(InPaths=Make Array("[0]"=@n2, "[1]"=@n3)))
   > Branch(Condition=DirectoryExists(InPath=@n4))
   n5 = GetBaseFilename(InPath=@n3, bRemovePath=true)
   > n6 = Execute Python Script!(
    execute
    PythonScript="import os\r\n\r\nmatches = []\r\nfiles = os.listdir(dir)\r\n\r\nfor f in files:\r\n    if f.startswith(file_name+'_Auto') or f.startswith(file_name+'_VRS_'):\r\n        matches.append(os.path.join(dir,f))"
    dir:String=@n4
    file_name:String=@n5
   ) -> (then, ReturnValue, matches:Array of Strings) {inputs [dir,file_name], outputs [matches]}
   > Return Node!(execute, Level Sequence Backups:Array of Strings=@n6.matches, Base Name:String=@n5)
  ]}
 Get Level Sequence Directory Path {nodes [
   n1 = Get Level Sequence Directory Path!() -> (then, Character:"String (by ref)", Category:String) {is=title K2Node_FunctionEntry}
   n2 = AnimatableCreature Find(AnimatableCreatures=Get AnimatableCreature(), Asset=@n1.Category, Variant=@n1)
   > n3 = Is Valid(InputObject=@n2)
   Return Node!(execute=@n3.Is Not Valid, Animation Directory:String)
   > Return Node(Animation Directory=Get Source Path!(self:Animatable Creature Object Reference=@n2) -> (Source Path:String))
  ]}
 Get Projects {nodes [
   Get Projects() {is=title K2Node_FunctionEntry}
   > Clear(TargetArray:Array of BP Animation Browser Project Object References=Get Local Projects() -> (Local Projects:Array of BP Animation Browser Project Object References))
   > n1 = GetAssetsByClass(
    self=GetAssetRegistry()
    ClassPathName=GetSoftClassTopLevelAssetPath!(SoftClassReference:Object Soft Class Reference=/AnimInEdTools/AnimationBrowser/ProjectSettings/BP_AnimationBrowser_Project.BP_AnimationBrowser_Project_C) -> (ReturnValue:Top Level Asset Path Structure)
    bSearchSubClasses=false
   )
   > n2 = For Each Loop(Array:Array of Asset Data Structures=@n1) -> (Array Element:Asset Data Structure)
   Return Node!(execute=@n2.Completed, Projects:Array of BP Animation Browser Project Object References=Get Local Projects())
   n3 = Break Asset Data!(AssetData=@n2) -> (PackageName, PackagePath:Name, AssetName, AssetClassPath:Top Level Asset Path Structure)
   > n4 = LoadAsset(AssetPath=GetBaseFilename(InPath="To String (Name)"(InName=@n3), bRemovePath=false))
   > n5 = Cast To BP_AnimationBrowser_Project!(execute, Object=@n4) -> (then, CastFailed, AsBP Animation Browser Project:BP Animation Browser Project Object Reference) {is=title K2Node_DynamicCast}
   > Add(
    TargetArray:Array of BP Animation Browser Project Object References=Get Local Projects()
    NewItem:"BP Animation Browser Project Object Reference (by ref)"=@n5
   )
  ]}
 Get Selected ListView Item {nodes [
   Get Selected ListView Item() {is=title K2Node_FunctionEntry}
   > n1 = Cast To Object_AnimationBrowser(Object=Get Selected Item(self=Get ListView()))
   > Return Node!(execute, Item:Object Animation Browser Object Reference=@n1)
   Return Node(execute=@n1.CastFailed)
  ]}
 Get Subfolders {nodes [
   Get Subfolders() {is=title K2Node_FunctionEntry}
   > n1 = Sequence()
   > Clear(TargetArray:Array of Strings=Get NewLocalVar() -> (NewLocalVar:Array of Strings))
   n2 = Get AnimatableCreature()
   n3 = Get Current Asset()
   n4 = Get Current Variation()
   n5 = Get Current Variation()
   > n6 = For Each Loop(Array:Array of Strings=Select(
     Option 0:Array of Strings=Make Array!("[0]"=@n5) -> (Array)
     Option 1:Array of Strings=Animatable Creature Get Variations(AnimatableCreatures=@n2, Asset Name=@n3, Add All Option=false)
     Index="Equal, Case Insensitive (String)"(A=@n4, B=All)
    ) -> (ReturnValue:Array of Strings)) -> (Array Element:String)
   n7 = AnimatableCreature Find(AnimatableCreatures=Get AnimatableCreature(), Asset=Get Current Asset(), Variant=@n6)
   > Is Valid(InputObject=@n7)
   > n8 = Branch(Condition=Get isShotgridEnabled())
   > n9 = Cast To AnimatableCreature_ShotGrid(Object=@n7)
   > n10 = Branch(Condition=Get SG Enabled(self=@n9))
   n11 = Get Plugin_Combobox()
   > n12 = Branch(Condition="Not Equal Exactly (String)"(A=GetSelectedOption(self=@n11), B=All))
   > n13 = Execute Python Script!(
    execute
    PythonScript="import AnimationBrowser as AMB\r\n\r\nname = f\"{asset}_{variant}\"\r\n\r\nfeatures = AMB.get_primary_parent_feature_filter(name)"
    asset:String=Get Asset(self=@n9)
    variant:String=Get Variant(self=@n9)
   ) -> (then, ReturnValue, features:Array of Strings) {inputs [asset,variant], outputs [features]}
   > n14 = Branch(Condition=Is Not Empty(TargetArray:Array of Strings=@n13.features))
   > Return Node!(execute, Subfolders:Array of Strings=@n13.features)
   n15 = Branch(execute=[@n14.else, @n9.CastFailed, @n10.else, @n12.else, @n8.else], Condition=Get Populate Subfolders From Disk(self=@n7))
   > n16 = Get Subfolder Root!(execute, self:EU Animation Naming Convention Object Reference=Get Naming Convention(self=@n7), Path:String=Get Animation Path(self=@n7)) -> (then, Root Path:String) {is=title K2Node_CallFunction}
   > n17 = ListAssets!(execute, DirectoryPath:String=@n16, bRecursive:bool=false, bIncludeFolder:bool=true) -> (then, ReturnValue:Array of Strings)
   > n18 = For Each Loop(Array=@n17)
   > n19 = DoesDirectoryExist(DirectoryPath=@n18)
   > Branch(Condition=@n19)
   > Add Unique(TargetArray=Get NewLocalVar(), NewItem=GetCleanFilename(InPath=Split(SourceString=@n18, InStr=/, SearchDir=FromEnd)))
   n20 = For Each Loop(Exec=@n15.else, Array=Get Subfolders(self=@n7) {is=title K2Node_CallFunction})
   > Add Unique(TargetArray=Get NewLocalVar(), NewItem=@n20)
   Return Node(execute=@n1.then_1, Subfolders=Get NewLocalVar())
  ]}
 Get Subfolders From Disk {nodes [
   n1 = Get Subfolders From Disk!() -> (then, Animatable Creature:Animatable Creature Object Reference) {is=title K2Node_FunctionEntry}
   > n2 = Sequence()
   > n3 = Get Subfolder Root(self=Get Naming Convention(self=@n1), Path=Get Animation Path(self=@n1))
   > n4 = ListAssets(DirectoryPath=@n3, bRecursive=false, bIncludeFolder=true)
   > n5 = For Each Loop(Array=@n4)
   > n6 = DoesDirectoryExist(DirectoryPath=@n5)
   > Branch(Condition=@n6)
   > Add Unique(
    TargetArray=Get Subfolders() -> (Subfolders:Array of Strings) {is=title K2Node_VariableGet}
    NewItem=GetCleanFilename(InPath=Split(SourceString=@n5, InStr=/, SearchDir=FromEnd))
   )
   Return Node(execute=@n2.then_1, Subfolders=Get Subfolders())
  ]}
 Is ComboBox All {nodes [
   n1 = Is ComboBox All!() -> (then, CB:"Searchable Combo Box (String) Object Reference") {is=title K2Node_FunctionEntry}
   > Return Node!(execute, "  ":bool="Equal Exactly (String)"(A=GetSelectedOption(self=@n1), B=All))
  ]}
 IsAC_SG_Enabled {nodes [
   n1 = IsAC_SG_Enabled!() -> (then, AC:Animatable Creature Object Reference) {is=title K2Node_FunctionEntry}
   > n2 = Cast To AnimatableCreature_ShotGrid(Object=@n1)
   > Return Node!(execute, isSG Enabled:bool=Get SG Enabled(self=@n2))
   Return Node(execute=@n2.CastFailed, isSG Enabled=false)
  ]}
 Load Additional Project Options {nodes [
   Load Additional Project Options() {is=title K2Node_FunctionEntry}
   > Clear(TargetArray:Array of EU Animation Browser Project Options Object References=Get Active Project Options())
   > n1 = For Each Loop(Array:Array of EU Animation Browser Project Options Soft Class References=Get Project Options!(self:BP Animation Browser Project Object Reference=Get Current Project()) -> (Result:Array of EU Animation Browser Project Options Soft Class References) {is=title K2Node_CallFunction}) -> (Array Element:EU Animation Browser Project Options Soft Class Reference)
   Load Toolbar Buttons(execute=@n1.Completed)
   n2 = LoadClassAsset_Blocking(execute=@n1.LoopBody, AssetClass=@n1)
   > n3 = Construct Object from Class(Class=@n2) -> (ReturnValue:Object Reference)
   > Branch(Condition=AND Boolean(
     A=IsValid(Object=@n3)
     B="Not Equal (Object)"(A:Object Reference=@n3, B:Object Reference) -> (ReturnValue:bool=~) {is=title K2Node_PromotableOperator}
    ))
   > n4 = Cast To EU_AnimationBrowser_ProjectOptions!(execute, Object=@n3) -> (then, CastFailed, AsEU Animation Browser Project Options:EU Animation Browser Project Options Object Reference) {is=title K2Node_DynamicCast}
   > Add(
    TargetArray:Array of EU Animation Browser Project Options Object References=Get Active Project Options()
    NewItem:"EU Animation Browser Project Options Object Reference (by ref)"=@n4
   )
   > Set Active Animation Browser!(
    execute
    Active Animation Browser:Animation Browser Object Reference=Self-Reference()
    self:EU Animation Browser Project Options Object Reference=@n4
   ) -> (then, Output_Get:Animation Browser Object Reference)
   > ProjectSelected(self:EU Animation Browser Project Options Object Reference=@n4)
  ]}
 Load ContextMenu Additional Project Options {nodes [
   n1 = Load ContextMenu Additional Project Options!() -> (then, Context:Name) {is=title K2Node_FunctionEntry}
   > n2 = For Each Loop(Array:Array of EU Animation Browser Project Options Object References=Get Active Project Options()) -> (Array Element:EU Animation Browser Project Options Object Reference)
   Return Node!(execute=@n2.Completed, Had Options:bool=Get local_hadOptions() -> (local_hadOptions:bool))
   n3 = Get Context Menu Options(self=@n2)
   > n4 = For Each Loop!(Exec, Array:Array of S AB Context Menu Options Structures=@n3) -> (
    LoopBody
    Array Element_Enabled_2_#4:bool=~
    Array Element_DisplayName_6_#5:String
    Array Element_FunctionToCall_10_#6:String
    Array Element_Category_18_#7:String
    Array Element_Reference_15_#8:EUW Context Menu Icon Object Reference
    Array Element_Context_13_#9:Name
    Array Element_Icon_21_#10:Texture 2D Object Reference
    Array Index
    Completed
   )
   > Branch(Condition="Equal (Name)"(A=@n4.Array Element_Context_13_#9, B=@n1))
   > Branch(Condition=@n4)
   > n5 = Add Icon Item(
    self=Get EUW_ContextMenu()
    Event=Create Event()
    Category=@n4.Array Element_Category_18_#7
    Name=@n4.Array Element_DisplayName_6_#5
    Font Size=10
    Icon=@n4.Array Element_Icon_21_#10
   )
   > Set members in S AB Context Menu Options!(
    execute:Exec
    StructRef:"S AB Context Menu Options Structure (by ref)"="Get (a ref)"(Array:Array of S AB Context Menu Options Structures=@n3, Dimension 1=@n4.Array Index) -> (Output:"S AB Context Menu Options Structure (by ref)")
    Reference_15_#8:EUW Context Menu Icon Object Reference=@n5
   ) -> (then:Exec, StructOut:"S AB Context Menu Options Structure (by ref)") {is=title K2Node_SetFieldsInStruct}
   > Set local_hadOptions(local_hadOptions:bool=true) -> (Output_Get:bool)
  ]}
 Load Dynamic Context Menu {nodes [
   n1 = Load Dynamic Context Menu!() -> (then, Context:Name, FocusButton:Widget Object Reference) {is=title K2Node_FunctionEntry}
   > Set local_focus(local_focus:Widget Object Reference=Select(
     Option 0:Widget Object Reference=@n1.FocusButton
     Option 1:Widget Object Reference=Get SettingsButton()
     Index="Equal (Object)"(A=@n1.FocusButton)
    ) -> (ReturnValue:Widget Object Reference)) -> (Output_Get:Widget Object Reference)
   n2 = Get EUW_ContextMenu()
   n3 = "To String (Name)"(InName=@n1)
   n4 = Get Name!(self:EUW Context Menu Object Reference=@n2) -> (Name)
   > n5 = Branch(Condition=AND Boolean(A=IsVisible(self=@n2), B="Equal Exactly (String)"(A=@n4, B=@n3)))
   Show(execute=[@n7.else, @n5.then], self=Get EUW_ContextMenu(), Show=false, Name=@n3)
   n6 = Load ContextMenu Additional Project Options(execute=@n5.else, Context=@n1) {is=title K2Node_CallFunction}
   > n7 = Branch(Condition=@n6)
   n8 = GetCachedGeometry(self=Get local_focus() -> (local_focus:Widget Object Reference))
   n9 = GetLocalSize(Geometry=@n8)
   > Show(
    self=Get EUW_ContextMenu()
    Show=true
    Mouse Position="Local to Absolute (Coordinates)"(Geometry=@n8, LocalCoordinate_X=@n9, LocalCoordinate_Y=@n9.ReturnValue_Y)
    Name=@n3
   )
  ]}
 Load Save Game From Slot {nodes [
   Load Save Game From Slot() {is=title K2Node_FunctionEntry}
   > ClearChildren(self=Get ModalLayer())
   > n1 = LoadGameFromSlot!(execute, SlotName:String=AnimationBrowser, UserIndex:int=0) -> (then, ReturnValue:Save Game Object Reference)
   > n2 = Cast To SG_AnimationBrowser!(execute, Object=@n1) -> (then, CastFailed, AsSG Animation Browser:SG Animation Browser Object Reference) {is=title K2Node_DynamicCast}
   > n3 = Set Save Game(Save Game=@n2)
   > n4 = LoadAsset_Blocking(Asset=Get Project!(self:SG Animation Browser Object Reference=@n3) -> (Project:BP Animation Browser Project Soft Object Reference))
   > n5 = Cast To BP_AnimationBrowser_Project(Object=@n4)
   > n6 = Set Current Project(Current Project:BP Animation Browser Project Object Reference=@n5) -> (Output_Get:BP Animation Browser Project Object Reference)
   > Set Project!(execute, self, Project:BP Animation Browser Project Object Reference=@n6, Save Settings:bool=false) -> (then) {is=title K2Node_CallFunction}
   n7 = Get Search!(self:SG Animation Browser Object Reference=Get Save Game()) -> (Search:Text)
   > Set Text(Text=@n7, self:Text Box Object Reference=Get SearchTextBox())
   > Set Current Filter(Current Filter="To String (Text)"(InText=@n7))
   > SetSelectedOption(self=Get Asset_ComboBox(), Option=Get Asset(self:SG Animation Browser Object Reference=Get Save Game()))
   > SetSelectedOption(self=Get Plugin_Combobox(), Option=Get Variant(self:SG Animation Browser Object Reference=Get Save Game()))
   > SetSelectedOption(self=Get Subfolder_Combobox(), Option=Get Asset(self=Get Save Game()))
   > SetSelectedOption(self=Get Subfolder_Combobox(), Option=Get Subfolder!(self:SG Animation Browser Object Reference=Get Save Game()) -> (Subfolder:String))
   > Set Developer Mode!(execute, Developer Mode=Get Developer Mode!(self:SG Animation Browser Object Reference=Get Save Game()) -> (Developer Mode)) -> (then, Output_Get)
   n8 = Get ShotGrid Enabled Status!(self:SG Animation Browser Object Reference=Get Save Game()) -> (ShotGrid Enabled Status:int=0)
   > n9 = Branch(Condition="Equal (Integer)"(A=@n8, B=-1))
   Set isShotgridEnabled(execute=@n9.else, isShotgridEnabled:bool="To Boolean (Integer)"!(InInt:int=@n8) -> (ReturnValue:bool)) -> (Output_Get:bool)
   n10 = CreateSaveGameObject(execute=[@n5.CastFailed, @n2.CastFailed], SaveGameClass=&SG_AnimationBrowser_C)
   > Set Save Game(Save Game=@n10)
   > Save Settings()
  ]}
 Load Toolbar Buttons {nodes [
   Load Toolbar Buttons() {is=title K2Node_FunctionEntry}
   > ClearChildren(self=Get OptionToolbar() -> (OptionToolbar:Editor Utility Scroll Box Object Reference))
   > Reverse!(execute, TargetArray:Array of EU Animation Browser Project Options Object References=Get Active Project Options()) -> (then) {is=title K2Node_CallArrayFunction}
   > n1 = For Each Loop!(Exec, Array:Array of EU Animation Browser Project Options Object References=Get Active Project Options()) -> (LoopBody, Array Element:EU Animation Browser Project Options Object Reference, Array Index, Completed)
   > n2 = For Each Loop(Array:Array of S AB Toolbar Options Structures=Get Custom Toolbar!(self:EU Animation Browser Project Options Object Reference=@n1) -> (Custom Toolbar:Array of S AB Toolbar Options Structures)) -> (Array Element:S AB Toolbar Options Structure)
   > Branch(Condition=Break S AB Toolbar Options!(S_AB_ToolbarOptions:"S AB Toolbar Options Structure (by ref)"=@n2) -> (
     Enabled_2_#4:bool=~
     Context_13_#9:Name
     Tooltip_23_#5:String
     FunctionToCall_10_#6:String
     ButtonIcon_22_#10:Texture 2D Object Reference
    ) {is=title K2Node_BreakStruct})
   n3 = Break S AB Toolbar Options(S_AB_ToolbarOptions=@n2) -> (Enabled_2_#4=~)
   n4 = Get contexts() -> (contexts:Array of Names)
   > Branch(Condition=NOT Boolean(A=Contains Item(TargetArray=@n4, ItemToFind=@n3.Context_13_#9)))
   > Add(TargetArray:Array of Names=Get contexts(), NewItem:"Name (by ref)"=@n3.Context_13_#9)
   > n5 = Construct Object from Class(Class=&/EUW_AnimBrowserToolbarButton.EUW_AnimBrowserToolbarButton_C) -> (ReturnValue:EUW Anim Browser Toolbar Button Object Reference)
   > AddChild(self=Get OptionToolbar(), Content=@n5)
   > Update!(
    execute
    self:EUW Anim Browser Toolbar Button Object Reference=@n5
    Option:S AB Toolbar Options Structure=@n2
    ProjectOptions:EU Animation Browser Project Options Object Reference=@n1
   ) -> (then)
  ]}
 Modal Dialog Displayed {nodes [
   n1 = Modal Dialog Displayed!() -> (then, Modal:EUW Modal Dialog Object Reference, Widget:User Widget Object Reference) {is=title K2Node_FunctionEntry}
   > n2 = For Each Loop(Array:Array of EU Animation Browser Project Options Object References=Get Active Project Options()) -> (Array Element:EU Animation Browser Project Options Object Reference)
   > ModalDisplayed!(
    execute
    self:EU Animation Browser Project Options Object Reference=@n2
    Modal:EUW Modal Dialog Object Reference=@n1
    Widget:User Widget Object Reference=@n1.Widget
   ) -> (then)
  ]}
 Move Binding To Level Sequence {nodes [
   Move Binding To Level Sequence!() -> (then, Source Level Sequence:Movie Scene Sequence Object Reference, Destination Level Sequence:Movie Scene Sequence Object Reference) {is=title K2Node_FunctionEntry}
   n1 = GetSelectedBindings()
   > SelectBindings!(execute, ObjectBindings:Array of Movie Scene Binding Proxy Structures=@n1) -> (then)
   > n2 = For Each Loop(Array:Array of Movie Scene Binding Proxy Structures=@n1) -> (Array Element:Movie Scene Binding Proxy Structure)
   > n3 = Sequence()
   > n4 = Set Destination Binding(Destination Binding:Movie Scene Binding Proxy Structure=@n2) -> (Output_Get:Movie Scene Binding Proxy Structure)
   > n5 = GetId!(execute, InBinding:"Movie Scene Binding Proxy Structure (by ref)"=@n4) -> (then, ReturnValue:Guid Structure)
   > n6 = FindBindingById!(
    execute
    Sequence:Movie Scene Sequence Object Reference=Get Source Level Sequence() -> (Source Level Sequence:Movie Scene Sequence Object Reference)
    BindingId:Guid Structure=@n5
   ) -> (then, ReturnValue:Movie Scene Binding Proxy Structure)
   > "Set Source  Binding"("Source  Binding":Movie Scene Binding Proxy Structure=@n6) -> (Output_Get:Movie Scene Binding Proxy Structure)
   Branch(
    execute=@n3.then_1
    Condition=IsValid!(InBinding:"Movie Scene Binding Proxy Structure (by ref)"="Get Source  Binding"() -> ("Source  Binding":Movie Scene Binding Proxy Structure)) -> (ReturnValue)
   )
   > n7 = Sequence!(execute) -> (then_0, then_1, then_2:Exec, then_3:Exec)
   > n8 = GetTracks!(execute, InBinding:"Movie Scene Binding Proxy Structure (by ref)"="Get Source  Binding"()) -> (then, ReturnValue:Array of Movie Scene Track Object References)
   > n9 = CopyTracks!(
    execute
    self:Level Sequence Editor Subsystem Object Reference=Get LevelSequenceEditorSubsystem()
    Tracks:Array of Movie Scene Track Object References=@n8
   ) -> (then, ExportedText:String)
   RefreshCurrentLevelSequence!(execute=@n7.then_3) -> (then)
   n10 = GetTracks(execute=@n7.then_1, InBinding=Get Destination Binding() -> (Destination Binding:Movie Scene Binding Proxy Structure))
   > n11 = For Each Loop(Array:Array of Movie Scene Track Object References=@n10) -> (Array Element:Movie Scene Track Object Reference)
   > RemoveTrack!(
    execute
    InBinding:"Movie Scene Binding Proxy Structure (by ref)"=Get Destination Binding()
    TrackToRemove:Movie Scene Track Object Reference=@n11
   ) -> (then)
   n12 = Get Destination Binding()
   PasteTracks!(
    execute=@n7.then_2
    self:Level Sequence Editor Subsystem Object Reference=Get LevelSequenceEditorSubsystem()
    TextToImport:String=@n9
    PasteTracksParams:Movie Scene Paste Tracks Params Structure=Make Movie Scene Paste Tracks Params!(
     Sequence:Movie Scene Sequence Object Reference=Get Destination Level Sequence() -> (Destination Level Sequence:Movie Scene Sequence Object Reference)
     Bindings:Array of Movie Scene Binding Proxy Structures=Make Array("[0]":Movie Scene Binding Proxy Structure=@n12) -> (Array:Array of Movie Scene Binding Proxy Structures)
     ParentFolder:Movie Scene Folder Object Reference
     Folders:Array of Movie Scene Folder Object References
    ) -> (MovieScenePasteTracksParams:Movie Scene Paste Tracks Params Structure) {is=title K2Node_MakeStruct}
   ) -> (then, OutTracks:Array of Movie Scene Track Object References, ReturnValue:bool)
  ]}
 OnMouseButtonDown {nodes [
   OnMouseButtonDown!() -> (then, MyGeometry:Geometry Structure, MouseEvent:"Pointer Event Structure (by ref)") {is=title K2Node_FunctionEntry}
   > Show(self=Get EUW_ContextMenu(), Show=false)
   > Return Node!(execute, ReturnValue:Event Reply Structure=Unhandled())
  ]}
 Open Animation Montage {nodes [
   n1 = Open Animation Montage!() -> (then, Context Item:Editor Utility Widget Object Reference, Slot Widget:Editor Utility Widget Object Reference) {is=title K2Node_FunctionEntry}
   > n2 = Cast To EUW_ContextMenu_Icon(Object=@n1.Slot Widget)
   n3 = Get String Data(self=@n2)
   > n4 = LoadAsset_Blocking(Asset=To Soft Object Reference(SoftObjectPath=MakeSoftClassPath(PathString=@n3)))
   > OpenEditorForAssets(self=Get AssetEditorSubsystem(), Assets=Make Array("[0]":Object Reference=@n4) -> (Array:Array of Object References), OpenedMethod=Edit)
  ]}
 Open Default Animation Map {nodes [
   Open Default Animation Map() {is=title K2Node_FunctionEntry}
   n1 = Get Levels(self=Get Current Project())
   n2 = "Get (a copy)"(Array=@n1)
   n3 = Length(TargetArray=@n1)
   > n4 = Branch(Condition=AND Boolean(A="integer >= integer"(A=@n3, B=1), B=IsValidSoftObjectReference(SoftObjectReference=@n2)))
   > n5 = Set Level Path(Level Path:String=To Soft Object Path(SoftObjectReference=@n2)) -> (Output_Get:String)
   n6 = DoesAssetExist(execute=[@n9.then, @n5.then], AssetPath=Get Level Path() -> (Level Path:String))
   > n7 = Branch(Condition=@n6)
   > n8 = LoadMap!(execute, Filename:String=Get Level Path()) -> (then, ReturnValue:World Object Reference)
   > Return Node!(execute, Map:World Object Reference=@n8)
   Return Node(execute=@n7.else)
   n9 = Set Level Path(
    execute=@n4.else
    Level Path=Combine(InPaths=Make Array!("[0]":String=&AnimationBrowser/, "[1]":String=Assets/AnimationBrowser_DefaultLevel) -> (Array:Array of Strings))
   )
  ]}
 Open Item {nodes [
   n1 = Open Item!() -> (then, Item:Object Animation Browser Object Reference, Level Path:String) {is=title K2Node_FunctionEntry}
   n2 = Get(self=@n1)
   > Set Anim Sequence(Anim Sequence:Soft Object Reference=@n2) -> (Output_Get:Soft Object Reference)
   > n3 = Branch(Condition=IsValidSoftObjectReference(SoftObjectReference=@n2.Data_LevelSequence_32_#2))
   > n4 = LoadAsset_Blocking(Asset=@n2.Data_LevelSequence_32_#2)
   > Is Valid(InputObject=@n4)
   > n5 = Cast To LevelSequence(Object=@n4)
   n6 = Get Level Path()
   > n7 = Branch(Condition="Equal (Integer)"(A=Len(S=@n6)))
   > n8 = Get Default Map for Level Sequence(Level Sequence=@n5) {is=title K2Node_CallFunction}
   > n9 = Open Sequence in Level(LevelSequence=@n5, Level Path=@n8)
   Add LevelSequence to Recents(
    execute=[@n10.then, @n9.then]
    LevelSequence="To String (SoftObjectReference)"(SoftObjectReference=@n2.Data_LevelSequence_32_#2)
   )
   n10 = Open Sequence in Level(execute=@n7.else, LevelSequence=@n5, Level Path=Get Level Path())
   n11 = Find Animatable Creature Using Name Semantics(execute=@n3.else, Asset Reference=Get Anim Sequence() -> (Anim Sequence:Soft Object Reference))
   > n12 = Set L_AC FIle(L_AC FIle:Animatable Creature Object Reference=@n11) -> (Output_Get:Animatable Creature Object Reference)
   > n13 = Is Valid(InputObject=@n12)
   n14 = AnimatableCreature Find(
    AnimatableCreatures=Get AnimatableCreature()
    Asset=GetSelectedOption(self=Get Asset_ComboBox())
    Variant=GetSelectedOption(self=Get Plugin_Combobox())
   )
   n15 = Branch(
    execute=@n13.Is Not Valid
    Condition=IsAC_SG_Enabled!(AC:Animatable Creature Object Reference=@n14) -> (isSG Enabled:bool) {is=title K2Node_CallFunction}
   )
   > n16 = Set L_AC FIle(L_AC FIle=@n14)
   n17 = Branch(
    execute=[@n13.Is Valid, @n16.then]
    Condition=Get Control Rig Backward Solve Enabled!(self:Animatable Creature Object Reference=@n12) -> (Control Rig Backward Solve Enabled:bool)
   )
   > n18 = Show Message Dialog(
    Title=MakeLiteralText(Value="LOCTEXT(\"Level Sequence does not exist\")")
    Message=MakeLiteralText(Value="LOCTEXT(\"Do you want to create a new LevelSequence by baking animation to the rig?\")")
    MessageType=YesNo
    DefaultValue=No
    MessageCategory=Warning
   )
   > n19 = Switch on EAppReturnType(Selection=@n18)
   n20 = Get L_AC FIle() -> (L_AC FIle:Animatable Creature Object Reference)
   n21 = Is Valid(exec=@n19.Yes, InputObject=@n20)
   Show Message Dialog(
    execute=@n21.Is Not Valid
    Title=MakeLiteralText(Value="LOCTEXT(\"Failed to invert animation \")")
    Message=MakeLiteralText(Value="LOCTEXT(\"Cannot find a valid source location for the Level Sequence\")")
    MessageType=Ok
    DefaultValue=Ok
    MessageCategory=Error
   )
   n22 = LoadAsset_Blocking(execute=@n21.Is Valid, Asset=Get Anim Sequence())
   > n23 = Is Valid(InputObject=@n22)
   PrintString(
    execute=[@n24.CastFailed, @n23.Is Not Valid]
    InString=Error: Failed to resolve Anim Sequence path to create new Level Sequence
    bPrintToScreen=false
   )
   n24 = Cast To AnimSequence(execute=@n23.Is Valid, Object=@n22)
   > Set L_AS(L_AS:Anim Sequence Object Reference=@n24) -> (Output_Get:Anim Sequence Object Reference)
   > n25 = Load Asset Blocking(InRef=Get Control Rig(self=Get L_AC FIle()), bForceLoad=true)
   > n26 = LoadClassAsset_Blocking(AssetClass=Get Sequencer Actor(self=Get L_AC FIle()))
   > n27 = Cast To AnimatableCreature_ShotGrid(Object=Get L_AC FIle())
   > n28 = Branch(Condition=Get SG Enabled(self=@n27))
   n29 = GetObjectName(Object=Get L_AS() -> (L_AS:Anim Sequence Object Reference))
   > n30 = ResolveArtifactByName!(
    execute
    Asset Name:String=Append!(A=Anim_, B=Split(SourceString=@n29, InStr=_, SearchDir=FromStart).RightS) -> (ReturnValue)
    artifact_name:String=level_sequence
    __WorldContext:Object Reference
   ) -> (then, ReturnValue:String)
   > n31 = Branch(Condition=IsEmpty(InString=@n30))
   Set L_pkgPath(execute=@n31.else, L_pkgPath:String=GetPath(InPath=@n30)) -> (Output_Get:String)
   > n32 = Set L_assetName(L_assetName:String=GetBaseFilename(InPath=@n30, bRemovePath=true)) -> (Output_Get:String)
   Create New Level Sequence And Bake Animation To Rig!(
    execute=[@n39.then, @n32.then]
    self
    Asset Name:String=Get L_assetName() -> (L_assetName:String)
    Package Path:String=Get L_pkgPath() -> (L_pkgPath:String)
    bReduceKeys:bool=true
    Tolerance:float=0
    AnimSequence:Anim Sequence Base Object Reference=Get L_AS()
    ControlRigClass:Control Rig Asset Strong Reference Structure=@n25
    Actor Class:Object Class Reference=@n26
   ) -> (then) {is=title K2Node_CallFunction}
   n33 = GetBaseFilename(InPath="To String (SoftObjectReference)"(SoftObjectReference=Get Anim Sequence()), bRemovePath=false)
   n34 = Replace(SourceString=@n33, From=Get Animation Path(self=@n20), To=Get Source Path(self=@n20))
   Set L_assetName(execute=[@n28.else, @n27.CastFailed], L_assetName=GetBaseFilename(InPath=@n34, bRemovePath=true))
   > Set L_pkgPath(L_pkgPath=GetPath(InPath=@n34))
   n35 = Get L_AC FIle()
   n36 = Contains(SearchIn=Get Animation Path(self=@n35), Substring="{")
   n37 = Contains(SearchIn=Get Source Path(self=@n35), Substring="{")
   > n38 = Branch(Condition=NOT Boolean(A=OR Boolean(A=@n37, B=@n36)))
   > n39 = Branch(Condition=NOT Boolean(A="Equal Exactly (String)"(A=@n34, B=@n33)))
   n40 = "To String (SoftObjectReference)"(SoftObjectReference=Get Anim Sequence())
   n41 = Get Animation Path(self=Get L_AC FIle())
   Show Message Dialog(
    execute=@n39.else
    Title=MakeLiteralText(Value="LOCTEXT(\"Name mapping failed\")")
    Message="To Text (String)"(InString=Append!(
      A="When attempting to remap from Anim Path to Source Path, the result did not change. This likely means that the AnimSequence that we're working on doesn't match the Anim Path in the AC file, therefore, we cannot do a search and replace to get the destination Level Sequence path.\r\n\r\nAnimation Path: "
      B=@n41
      C:String="\r\nAnim Sequence Asset Path: "
      D:String=@n40
     ) -> (ReturnValue))
    MessageType=Ok
    DefaultValue=Ok
    MessageCategory=Error
   )
   n42 = Get L_AS()
   n43 = Get Plugin_Combobox()
   n44 = Get Subfolder_Combobox()
   n45 = "To String (SoftObjectReference)"(SoftObjectReference=Get Anim Sequence())
   n46 = Get Asset_ComboBox()
   n47 = Get L_AC FIle()
   n48 = Execute Python Script!(
    execute=@n38.else
    PythonScript="## THIS IS NOT JUST A TOKENIZER.\r\n## This script takes in the final Anim Sequence asset path(input_asset_name), along with an incomplete token map(input_tokens) from the input.\r\n## Then, we try to make the token map more \"complete\" based on the information from the final Anim Sequence asset path and looking at how\r\n## the AC file defines the asset path. Once we've rebuilt the token map as best we could, we try to tokenize the\r\n## result package path(input_package_path) in order to output the desired Level Sequence Asset Path.\r\n\r\nimport re\r\nimport unreal\r\n\r\ndef log_error(message: str):\r\n    unreal.log_error(message)\r\n    unreal.log_flush()\r\n\r\ntoken_map = dict()\r\nfor token in input_tokens:\r\n    tkn = \"{\" + token + \"}\"\r\n    token_map[tkn] = input_tokens[token]\r\n\r\nDEBUG = False\r\nPATH_CHR = \"/\"\r\n\r\ndef extract_tokens(sourceString: str):\r\n    results = set()\r\n    index = -1\r\n    while True:\r\n        index = sourceString.find('{', index + 1)\r\n        if index == -1:\r\n            break\r\n        endIndex = sourceString.find('}', index)\r\n        if endIndex == -1:\r\n            break\r\n        if index + 1 != endIndex:\r\n            results.add(sourceString[index:endIndex + 1])\r\n        index = endIndex\r\n    return results\r\n\r\ndef match_segment(assetSegment: str, tokenSegment: str):\r\n    \"\"\"\r\n    Match a single slash-delimited path segment against a token pattern, returning\r\n    a dict of token->value mappings. e.g.:\r\n      assetSegment = \"AS_MyPlugin_Handstand_Vertical_S\"\r\n      tokenSegment = \"AS_{Plugin}_{AnimationName}\"\r\n      returns {\"{Plugin}\": \"MyPlugin\", \"{AnimationName}\": \"Handstand_Vertical_S\"}\r\n\r\n    Earlier tokens in the segment are non-greedy (match up to the next underscore).\r\n    The final token is greedy and absorbs all remaining content including underscores.\r\n    This means token ordering within a segment matters -- put fixed-length tokens\r\n    first and any value that may contain underscores (e.g. AnimationName) last.\r\n    \"\"\"\r\n    token_list = re.findall(r'\\{[^}]+\\}', tokenSegment)\r\n\r\n    pattern = re.escape(tokenSegment)\r\n    for i, token in enumerate(token_list):\r\n        escaped = re.escape(token)\r\n        name = token[1:-1]\r\n        if i == len(token_list) - 1:\r\n            pattern = pattern.replace(escaped, r'(?P<' + name + r'>.+)', 1)\r\n        else:\r\n            pattern = pattern.replace(escaped, r'(?P<' + name + r'>[^_]+)', 1)\r\n\r\n    match = re.fullmatch(pattern, assetSegment)\r\n    if not match:\r\n        log_error(\"Segment pattern match failed. Asset: '{}' Pattern: '{}'\".format(assetSegment, tokenSegment))\r\n        raise Exception(\"Segment pattern match failed. Asset: '{}' Pattern: '{}'\".format(assetSegment, tokenSegment))\r\n\r\n    return {'{' + k + '}': v for k, v in match.groupdict().items()}\r\n\r\ndef map_tokens(assetPath: str, tokenPath: str):\r\n    if assetPath.count('/') != tokenPath.count('/'):\r\n        log_error(\"Asset Path: {}\".format(assetPath))\r\n        log_error(\"Token Path: {}\".format(tokenPath))\r\n        log_error(\"Asset path does not have the same depth as the token path. Asset Path:{} Token Path:{}\".format(assetPath.count('/'), tokenPath.count('/')))\r\n        raise Exception(\"Asset path does not have the same depth as the token path. Asset Path:{} Token Path:{}\".format(assetPath.count('/'), tokenPath.count('/')))\r\n\r\n    assetSplit = assetPath.split('/')\r\n    tokenSplit = tokenPath.split('/')\r\n\r\n    for i, tokenSegment in enumerate(tokenSplit):\r\n        assetSegment = assetSplit[i] if i < len(assetSplit) else \"\"\r\n\r\n        if len(tokenSegment) == 0:\r\n            pass  ## Leading slash produces an empty segment, skip it\r\n        elif tokenSegment == assetSegment:\r\n            pass  ## Literal match, no tokens involved\r\n        elif not extract_tokens(tokenSegment):\r\n            log_error(\"Pattern matching failed on {} and {}. Non-Token entries must match identically.\".format(tokenSegment, assetSegment))\r\n            raise Exception(\"Pattern matching failed on {} and {}. Non-Token entries must match identically.\".format(tokenSegment, assetSegment))\r\n        else:\r\n            matches = match_segment(assetSegment, tokenSegment)\r\n            token_map.update(matches)\r\n\r\nmap_tokens(anim_sequence_name_path, input_asset_name)\r\n\r\nprint(\"\\t-----[RESULT TOKEN MAP]----- (Generated from the input tokens, and extrapolated with some pattern matching.)\")\r\nresult: str = input_package_path\r\nfor item in token_map:\r\n    result = result.replace(item, token_map[item])\r\n    print('{:>20}'.format(item), \" --> \", token_map[item])\r\nresult = result.rstrip('_')\r\nprint(\"Result Level Sequence Path: \", result)\r\n\r\nif result.find('{') >= 0:\r\n    raise Exception(\"End result still has unmapped token values for some reason.\")\r\n\r\nsplitPos = result.rindex('/')\r\nasset_name = result[splitPos+1:]\r\npackage_path = result[0:splitPos]\r\n\r\nif DEBUG:\r\n    print(\"PackagePath: \", package_path)\r\n    print(\"AssetName: \", asset_name)"
    inputTokens:Map of Strings to Strings=Get Path Parts Map(
     Asset=GetSelectedOption(self=@n46)
     Animation Name=GetBaseFilename(InPath=@n45, bRemovePath=true)
     Subfolder=GetSelectedOption(self=@n44)
     Plugin=GetSelectedOption(self=@n43)
    )
    inputAssetName:String=Get Animation Path(self=@n47)
    inputPackagePath:String=Get Source Path(self=@n47)
    animSequenceNamePath:String=GetBaseFilename(InPath=Get Object Path String(Object=@n42), bRemovePath=false)
   ) -> (then, ReturnValue, assetName:String, packagePath:String) {inputs [inputTokens, inputAssetName, inputPackagePath, animSequenceNamePath], outputs [assetName,packagePath]}
   > n49 = Branch(Condition=@n48)
   > Create New Level Sequence And Bake Animation To Rig(
    Asset Name=@n48.assetName
    Package Path=@n48.packagePath
    bReduceKeys=true
    Tolerance=0
    AnimSequence=Get L_AS()
    ControlRigClass=@n25
    Actor Class=@n26
   )
   Show Message Dialog(
    execute=@n49.else
    Title=MakeLiteralText(Value="LOCTEXT(\"Python Script Error\")")
    Message=MakeLiteralText(Value="LOCTEXT(\"Check the output log for errors.\")")
    MessageType=Ok
    DefaultValue=Ok
    MessageCategory=Error
   )
   n50 = GetDisplayName(Object=Get L_AC FIle())
   Show Message Dialog(
    execute=@n17.else
    Title=MakeLiteralText(Value="LOCTEXT(\"Cannot Invert Animation\")")
    Message="To Text (String)"(InString=Append!(
      A="AC File ["
      B=@n50
      C="] has \"Control Rig Backwards Solve Enabled\" set to false. It's unclear if the control rig actually supports backwards solve. Unable to proceed"
     ) -> (ReturnValue))
    MessageType=Ok
    DefaultValue=Ok
    MessageCategory=Warning
   )
   n51 = "To String (SoftObjectReference)"(SoftObjectReference=Get Anim Sequence())
   Show Message Dialog(
    execute=@n15.else
    Title=MakeLiteralText(Value="LOCTEXT(\"Could not find AC File\")")
    Message="To Text (String)"(InString=Append!(A="Could not find an appropriate AC file using name semantics for:\r\n", B=@n51) -> (ReturnValue))
    MessageType=Ok
    DefaultValue=Ok
    MessageCategory=Error
   )
   Add LevelSequence to Recents(LevelSequence="To String (SoftObjectReference)"())
   Get isShotgridEnabled()
  ]}
 Open Sequence in Level {nodes [
   Open Sequence in Level!() -> (then, LevelSequence:Level Sequence Object Reference, Level Path:String) {is=title K2Node_FunctionEntry}
   > n1 = Sequence!(execute) -> (then_0, then_1, then_2)
   > Set Stage Transform(Stage Transform:Transform) -> (Output_Get:Transform)
   n2 = Branch(execute=@n1.then_2, Condition=IsEmpty(InString=Get Current Level Path() -> (Current Level Path:String)))
   n3 = GetAssetsByClass(
    execute=[@n18.then, @n17.else, @n2.then]
    self=GetAssetRegistry()
    ClassPathName=GetSoftClassTopLevelAssetPath(SoftClassReference=&/DA_LevelTransforms.DA_LevelTransforms_C)
    bSearchSubClasses=false
   )
   > n4 = Branch(Condition=@n3.ReturnValue)
   > n5 = For Each Loop!(Exec, Array:Array of Asset Data Structures=@n3) -> (
    LoopBody
    Array Element_PackageName:Name
    Array Element_PackagePath:Name
    Array Element_AssetName:Name
    Array Element_AssetClassPath:Top Level Asset Path Structure
    Array Index
    Completed
   )
   > n6 = LoadAsset(AssetPath=GetBaseFilename(InPath="To String (Name)"(InName=@n5), bRemovePath=false))
   > n7 = Cast To DA_LevelTransforms!(execute, Object=@n6) -> (then, CastFailed, AsDA Level Transforms:DA Level Transforms Object Reference) {is=title K2Node_DynamicCast}
   > n8 = GetCurrentLevelName!(execute, WorldContextObject:Object Reference, bRemovePrefixString:bool=true) -> (then, ReturnValue:String)
   n9 = Get Level Name to Transform!(self:DA Level Transforms Object Reference=@n7) -> (Level Name to Transform:Map of Strings to Transforms)
   n10 = Find(TargetMap:Map of Strings to Transforms=@n9, Key=@n8) -> (Value:Transform)
   > n11 = Branch(Condition=@n10.ReturnValue)
   > n12 = Set Stage Transform(Stage Transform=@n10)
   n13 = OpenLevelSequence(
    execute=[@n15.then, @n12.then, @n5.Completed, @n4.else]
    LevelSequence=Get LevelSequence() -> (LevelSequence:Level Sequence Object Reference)
   )
   > Branch(Condition=@n13)
   > Add Stage Function(Transform=Get Stage Transform() -> (Stage Transform:Transform))
   > Return Node!(execute)
   n14 = Find(TargetMap=@n9, Key=Split(SourceString=Get Current Level Path(), InStr=., SearchDir=FromEnd).RightS)
   Branch(execute=@n11.else, Condition=@n14.ReturnValue)
   > n15 = Set Stage Transform(Stage Transform=@n14)
   n16 = DoesAssetExist(execute=@n2.else, AssetPath=Get Current Level Path())
   > n17 = Branch(Condition=@n16)
   > n18 = LoadMap(Filename=Get Current Level Path())
   n19 = Branch(execute=@n1.then_1, Condition=IsEmpty(InString=Get Level Path()))
   > n20 = GetCurrentLevelName(bRemovePrefixString=true)
   > Set Current Level Path(Current Level Path:String=SelectString(A=Get Default Map() {is=title K2Node_CallFunction}, B=Get Level Path(), bPickA=StartsWith(SourceString=@n20, InPrefix=Untitled_))) -> (Output_Get:String)
   Set Current Level Path(execute=@n19.else, Current Level Path=Get Level Path())
  ]}
 Paste Pose {nodes [
   n1 = Paste Pose!() -> (
    then
    Pose Asset:Soft Object Reference
    LevelSequence:Level Sequence Object Reference
    ControlRig Proxy Binding:"Control Rig Sequencer Binding Proxy Structure (by ref)"
   ) {is=title K2Node_FunctionEntry}
   > n2 = LoadAsset_Blocking(Asset=@n1)
   > Is Valid(InputObject=@n2)
   > n3 = Cast To ControlRigPoseAsset!(execute, Object=@n2) -> (then, CastFailed, AsControl Rig Pose Asset:Control Rig Pose Asset Object Reference) {is=title K2Node_DynamicCast}
   > n4 = GetControlRigs(LevelSequence=Get LevelSequence())
   > Branch(Condition=Is Not Empty(TargetArray:Array of Control Rig Sequencer Binding Proxy Structures=@n4))
   n5 = Get ControlRig Proxy Binding() -> (ControlRig Proxy Binding:"Control Rig Sequencer Binding Proxy Structure (by ref)")
   > n6 = Execute Python Script!(
    execute
    PythonScript="import unreal\r\n\r\nnum_keys = 0\r\nfor section in proxy.track.get_sections():\r\n    float_channels = section.get_channels_by_type(unreal.MovieSceneScriptingFloatChannel)\r\n    for channel in float_channels:\r\n        num_keys += channel.get_num_keys()\r\n\r\nresult = num_keys == 0\r\n"
    proxy:"Control Rig Sequencer Binding Proxy Structure (by ref)"=@n5
   ) -> (then, ReturnValue, result:bool=~) {inputs [proxy], outputs [result]}
   > Branch(Condition=@n6.result)
   n7 = Break Control Rig Sequencer Binding Proxy!(ControlRigSequencerBindingProxy=@n5) -> (Proxy:Movie Scene Binding Proxy Structure, ControlRig, Track)
   > SelectControls!(
    execute
    self:Control Rig Pose Asset Object Reference=@n3
    InControlRig:Control Rig Object Reference=@n7.ControlRig
    bDoMirror:bool=false
    bClearSelection:bool=true
   ) -> (then)
   > PastePose!(
    execute
    self:Control Rig Pose Asset Object Reference=@n3
    InControlRig:Control Rig Object Reference=@n7.ControlRig
    bDoKey:bool=true
    bDoMirror:bool=false
    bDoAdditive:bool=false
   ) -> (then)
   > ClearControlSelection!(execute, self:Control Rig Object Reference=@n7.ControlRig, bSetupUndo:bool=false) -> (then, ReturnValue:bool)
   "Get (a copy)"(Array:Array of Control Rig Sequencer Binding Proxy Structures=@n4) -> (Output:Control Rig Sequencer Binding Proxy Structure)
  ]}
 Remove Modal Dialog {nodes [
   Remove Modal Dialog() {is=title K2Node_FunctionEntry}
   > n1 = For Each Loop with Break(Array:Array of Widget Object References=GetAllChildren(self=Get ModalLayer()), Break=@n3.then) -> (Array Element:Widget Object Reference)
   > n2 = Cast To EUW_ModalDialog!(execute, Object=@n1) -> (then, CastFailed, AsEUW Modal Dialog:EUW Modal Dialog Object Reference) {is=title K2Node_DynamicCast}
   > n3 = RemoveFromParent(self=@n2)
  ]}
 Save Settings {nodes [
   Save Settings() {is=title K2Node_FunctionEntry}
   > SaveGameToSlot(SaveGameObject=Get Save Game(), SlotName=AnimationBrowser, UserIndex=0)
  ]}
 Set Default Map for Level Sequence {nodes [
   n1 = Set Default Map for Level Sequence!() -> (then, Level Sequence:Level Sequence Object Reference, Path to Map:String) {is=title K2Node_FunctionEntry}
   > n2 = Branch(Condition="Equal (Integer)"(A=Len(S=@n1.Path to Map)))
   > n3 = GetCurrentLevel(self=Get LevelEditorSubsystem())
   > n4 = LoadAsset(AssetPath=Split(SourceString=Get Object Path String(Object=@n3), InStr=:, SearchDir=FromEnd))
   n5 = LoadAsset(execute=@n2.else, AssetPath=@n1.Path to Map)
   n6 = GetPathNameForLoadedAsset(execute=[@n4.then, @n5.then], LoadedAsset=@n5)
   n7 = Get Level Sequence()
   > SetMetadataTag!(execute, Object:Object Reference=@n7, Tag:Name=Get Default Map Tag() -> (Default Map Tag=None), Value:String=@n6) -> (then)
   n8 = "To String (SoftObjectReference)"(SoftObjectReference=To Soft Object Reference(SoftObjectPath=GetSoftObjectPath!(Object=@n7) -> (ReturnValue:Soft Object Path Structure)))
   > n9 = SaveAsset!(execute, self:Editor Asset Subsystem Object Reference=Get EditorAssetSubsystem(), AssetToSave=@n8, bOnlyIfIsDirty=true) -> (then, ReturnValue)
   > n10 = Branch(Condition=@n9)
   > PrintString(InString=Append!(A=', B=@n6, C:String=' map set for ', D:String=@n8, E:String=' Level Sequence.) -> (ReturnValue))
   PrintString(
    execute=@n10.else
    InString=Append!(A=Could not save Level Sequence ', B=@n8, C="', please make sure it's available for checkout before submitting!") -> (ReturnValue)
   )
  ]}
 Set Project {nodes [
   n1 = Set Project!() -> (then, Project:BP Animation Browser Project Object Reference, Save Settings:bool=true) {is=title K2Node_FunctionEntry}
   > Set Current Project(Current Project=@n1)
   > Refresh Assets()
   > Update Character Combo Box() {is=title K2Node_CallFunction}
   > n2 = Sequence!(execute) -> (then_0, then_1, then_2, then_3:Exec)
   n3 = Get Asset_ComboBox()
   > Branch(Condition="integer > integer"(A=GetOptionCount!(self:"Searchable Combo Box (String) Object Reference"=@n3) -> (ReturnValue:int=0)))
   n4 = Get Asset_ComboBox()
   > SetSelectedIndex!(execute, self:"Searchable Combo Box (String) Object Reference"=@n4, Index:int=0) -> (then)
   > Set Current Asset(Current Asset=GetSelectedOption(self=@n4))
   Update Combo Box(
    execute=@n2.then_1
    Combo Box=Get Subfolder_Combobox()
    Items=Get Subfolders!() -> (Subfolders) {is=title K2Node_CallFunction}
    Default Item=SelectString(B=Get Animation Type Default Item(), bPickA=Get isShotgridEnabled())
    Add Blank=false
   )
   Branch(execute=@n2.then_2, Condition=Get Save Settings() -> (Save Settings:bool))
   n5 = Get Save Game()
   > Set Asset(Asset=GetSelectedOption(self=Get Asset_ComboBox()), self=@n5)
   > Set Variant(Variant=GetSelectedOption(self=Get Plugin_Combobox()), self=@n5)
   > Set Subfolder(Subfolder=GetSelectedOption(self=Get Subfolder_Combobox()), self=@n5)
   > Set Project(Project=Make Soft Reference(Input=Get Project!() -> (Project:BP Animation Browser Project Object Reference)), self=@n5) {is=title K2Node_VariableSet}
   > Save Settings() {is=title K2Node_CallFunction}
   Load Additional Project Options(execute=@n2.then_3) {is=title K2Node_CallFunction}
  ]}
 Set Project Dialog Button {nodes [
   n1 = Set Project Dialog Button!() -> (then, Project:BP Animation Browser Project Object Reference) {is=title K2Node_FunctionEntry}
   > Set Project(Project=@n1, Save Settings=true) {is=title K2Node_CallFunction}
   > ClearChildren(self=Get ModalLayer())
  ]}
 Settings Menu - Browse to Project {nodes [
   Settings Menu - Browse to Project!() -> (then, Context Item:Editor Utility Widget Object Reference, Slot Widget:Editor Utility Widget Object Reference) {is=title K2Node_FunctionEntry}
   n1 = Get Current Project()
   > SyncBrowserToObjects(AssetPaths=Make Array!("[0]"=GetSoftObjectPath!(Object=@n1) -> (ReturnValue_PathString:String)) -> (Array))
  ]}
 Settings Menu - Toggle ShotGrid Integration {nodes [
   n1 = Settings Menu - Toggle ShotGrid Integration!() -> (then, Context Item:Editor Utility Widget Object Reference, Slot Widget:Editor Utility Widget Object Reference) {is=title K2Node_FunctionEntry}
   n2 = NOT Boolean(A=Get isShotgridEnabled())
   > Set ShotGrid Enabled Status!(
    execute
    ShotGrid Enabled Status:int="To Integer (Boolean)"!(InBool:bool=@n2) -> (ReturnValue:int=0)
    self:SG Animation Browser Object Reference=Get Save Game()
   ) -> (then, Output_Get:int=0)
   > n3 = Set isShotgridEnabled(isShotgridEnabled=@n2)
   > n4 = Cast To EUW_ContextMenu_Icon(Object=@n1)
   > n5 = "SetText (Text)"(
    self=Get ButtonText!(self:EUW Context Menu Icon Object Reference=@n4) -> (ButtonText:Text Object Reference)
    InText=SelectText(A="LOCTEXT(\"Disable Shotgrid\")", B="LOCTEXT(\"Enable Shotgrid\")", bPickA=@n3)
   )
   Save Settings(execute=[@n5.then, @n4.CastFailed])
   SelectString(A=Disable Developer Mode, B=Enable Developer Mode, bPickA=Get Developer Mode!() -> (Developer Mode))
  ]}
 Settings Menu Browse to AC File {nodes [
   Settings Menu Browse to AC File!() -> (then, Context Item:Editor Utility Widget Object Reference, Slot Widget:Editor Utility Widget Object Reference) {is=title K2Node_FunctionEntry}
   n1 = AnimatableCreature Find(AnimatableCreatures=Get AnimatableCreature(), Asset=Get Current Asset(), Variant=Get Current Variation())
   > Is Valid(InputObject=@n1)
   > SyncBrowserToObjects(AssetPaths=Make Array("[0]"=GetSoftObjectPath(Object=@n1)))
  ]}
 Settings Menu Refresh {nodes [
   Settings Menu Refresh!() -> (then, Context Item:Editor Utility Widget Object Reference, Slot Widget:Editor Utility Widget Object Reference) {is=title K2Node_FunctionEntry}
   > Refresh Assets()
   > Update List View(
    Current Variation=Get Current Variation()
    Animation Type=Get Current Animation Type()
    Filter=Get Current Filter()
    Current Category=Get Current Asset()
    Use Cache=false
   )
  ]}
 Settings Menu Set Developer Mode {nodes [
   Settings Menu Set Developer Mode!() -> (then, Context Item:Editor Utility Widget Object Reference, Slot Widget:Editor Utility Widget Object Reference) {is=title K2Node_FunctionEntry}
   > n1 = Set Developer Mode(Developer Mode=NOT Boolean(A=Get Developer Mode()))
   > Set Developer Mode!(execute, Developer Mode=@n1, self:SG Animation Browser Object Reference=Get Save Game()) -> (then, Output_Get)
   > Save Settings()
  ]}
 Settings Menu Set Project {nodes [
   n1 = Settings Menu Set Project!() -> (then, Context Item:Editor Utility Widget Object Reference, Slot Widget:Editor Utility Widget Object Reference) {is=title K2Node_FunctionEntry}
   > n2 = Cast To EUW_ContextMenu_Icon(Object=@n1.Slot Widget)
   n3 = Get ButtonText(self=@n2)
   > Set Selected Name(Selected Name:String="To String (Text)"(InText=Get Text(self=@n3))) -> (Output_Get:String)
   > n4 = GetAssetsByClass(
    self=GetAssetRegistry()
    ClassPathName=GetSoftClassTopLevelAssetPath(SoftClassReference=&BP_AnimationBrowser_Project_C)
    bSearchSubClasses=false
   )
   > n5 = For Each Loop with Break(Array:Array of Asset Data Structures=@n4, Break=@n9.then) -> (Array Element:Asset Data Structure)
   n6 = Break Asset Data(AssetData=@n5)
   > n7 = LoadAsset(AssetPath=GetBaseFilename(InPath="To String (Name)"(InName=@n6), bRemovePath=false))
   > n8 = Cast To BP_AnimationBrowser_Project(Object=@n7)
   > Branch(Condition="Equal Exactly (String)"(A=Get Selected Name() -> (Selected Name:String), B=Get Name(self:BP Animation Browser Project Object Reference=@n8)))
   > n9 = Set Project(Project=@n8, Save Settings=true)
  ]}
 Shotgrid - Populate List from Context {nodes [
   n1 = Execute Python Script!(
    execute=[@n14.then, @n11.then, @n15.then]
    PythonScript="import AnimationBrowser as AMB\r\n\r\nout = AMB.get_anim_paths_for_feature_name(feature_name)"
    feature_name:String=GetSelectedOption(self=Get Subfolder_Combobox())
   ) -> (then, ReturnValue, out:Array of Strings) {inputs [feature_name], outputs [out]}
   n2 = For Each Loop!(Exec=@n16.then, Array:Array of Strings=@n1.out) -> (LoopBody, Array Element:String, Array Index, Completed)
   n3 = Add Data To List!(
    execute
    self
    Anim Sequence:Soft Object Reference=To Soft Object Reference!(SoftObjectPath_PathString:String=@n2) -> (ReturnValue)
    Level Sequence:Soft Object Reference
    Current User:String=Get Current User()
    Display Subfolder:bool="Equal Exactly (String)"(A=Get Current Animation Type(), B=Get Animation Type Default Item())
    AnimMontages:Array of Soft Object References
   ) -> (then, New Item:Object Animation Browser Object Reference) {is=title K2Node_CallFunction}
   n4 = Append(A=@n2, B=., C=Split(SourceString=@n2, InStr=/, SearchDir=FromEnd).RightS)
   Add Unique(execute=[@n3.then, @n2.LoopBody], TargetArray=Get Search AnimSequence Cache(), NewItem=@n4)
   n5 = IsEmpty(InString=@n2)
   > Branch(Condition=AND Boolean(A="Equal (Integer)"(A=@n2.Array Index), B=NOT Boolean(A=@n5)))
   > PrintString(InString=@n4)
   Shotgrid - Populate List from Context() {is=title K2Node_FunctionEntry}
   n6 = GetSelectedOption(self=Get Subfolder_Combobox())
   > n7 = Branch(Condition=IsEmpty(InString=@n6))
   > n8 = SetSelectedIndex(self=Get Subfolder_Combobox(), Index=0)
   n9 = Branch(execute=[@n8.then, @n7.else], Condition="Equal Exactly (String)"(A=@n6, B=None))
   n10 = Get Subfolder_Combobox()
   > RemoveOption!(execute, self:"Searchable Combo Box (String) Object Reference"=@n10, Option:String=None) -> (then, ReturnValue:bool)
   > n11 = SetSelectedIndex(self=@n10, Index=0)
   n12 = Branch(execute=@n9.else, Condition="Equal Exactly (String)"(A=@n6, B=All))
   n13 = Get Subfolder_Combobox()
   > RemoveOption(self=@n13, Option=All)
   > n14 = SetSelectedIndex(self=@n13, Index=0)
   n15 = Branch(execute=@n12.else, Condition=Contains Item(TargetArray:Array of Strings=Get Subfolders(), ItemToFind:"String (by ref)"=@n6))
   Return Node!(execute=@n15.else, Success:bool=false)
   Return Node(execute=@n2.Completed, Success=true)
   Get Subfolder_Combobox()
   n16 = Clear(execute=@n1.then, TargetArray:Array of Strings=Get Search AnimSequence Cache())
  ]}
 Show Currently Open Level Sequence Options {nodes [
   n1 = Show Currently Open Level Sequence Options!() -> (then, Show:bool=~, File Name:String) {is=title K2Node_FunctionEntry}
   > SetVisibility(
    self=Get ListHeaderSizeBox() -> (ListHeaderSizeBox:Size Box Object Reference)
    InVisibility=Select(Option 0:ESlateVisibility Enum=Collapsed, Option 1:ESlateVisibility Enum=Visible, Index=@n1) -> (ReturnValue:ESlateVisibility Enum)
   )
   > Set Text(
    Text="To Text (String)"(InString=Append!(A="Current Level Sequence >> ", B=@n1.File Name) -> (ReturnValue))
    self:Text Object Reference=Get RestoreTextBlock() -> (RestoreTextBlock:Text Object Reference)
   )
  ]}
 Update Combo Box {nodes [
   n1 = Update Combo Box!() -> (then, Combo Box:"Searchable Combo Box (String) Object Reference", Items:Array of Strings, Default Item:String, Add Blank:bool) {is=title K2Node_FunctionEntry}
   > n2 = Set Current Index(Current Index:int=GetSelectedIndex!(self:"Searchable Combo Box (String) Object Reference"=@n1) -> (ReturnValue:int=0)) -> (Output_Get:int=0)
   > Set Selected Item(Selected Item:String=GetOptionAtIndex!(
     self:"Searchable Combo Box (String) Object Reference"=Get Combo Box() -> (Combo Box:"Searchable Combo Box (String) Object Reference")
     Index:int=@n2
    ) -> (ReturnValue:String)) -> (Output_Get:String) {is=title K2Node_VariableSet}
   n3 = Get Combo Box()
   > n4 = For Loop(LastIndex=int - int(A=GetOptionCount(self=@n3) -> (ReturnValue=0)))
   n5 = Branch(
    execute=@n4.Completed
    Condition=Identical!(ArrayA:Array of Strings=Get Items() -> (Items:Array of Strings), ArrayB:Array of Strings=Get Old Items() -> (Old Items:Array of Strings)) -> (ReturnValue:bool) {is=title K2Node_CallArrayFunction}
   )
   ClearOptions(execute=@n5.else, self:"Searchable Combo Box (String) Object Reference"=Get Combo Box())
   > n6 = Sequence()
   > Branch(Condition=Get Add Blank() -> (Add Blank:bool))
   > AddOption!(execute, self:"Searchable Combo Box (String) Object Reference"=Get Combo Box(), Option:String) -> (then)
   n7 = Get Default Item() -> (Default Item:String)
   n8 = Branch(execute=@n6.then_1, Condition=IsEmpty(InString=@n7))
   AddOption(execute=@n8.else, self=Get Combo Box(), Option=@n7)
   n9 = For Each Loop(Exec=@n6.then_2, Array=Get Items())
   > AddOption(self=Get Combo Box(), Option=@n9)
   n10 = Get Combo Box()
   n11 = Get Current Index() -> (Current Index:int=0)
   n12 = Branch(
    execute=@n6.then_3
    Condition="Equal Exactly (String)"(A=GetOptionAtIndex(self=@n10, Index=@n11), B=Get Selected Item() -> (Selected Item:String) {is=title K2Node_VariableGet})
   )
   n13 = Get Combo Box()
   > AddOption(self=@n13, Option=Temp)
   n14 = Get Current Index() -> (Current Index=0)
   > SetSelectedIndex(self=@n13, Index=int + int(A=@n14))
   > SetSelectedIndex(self=@n13, Index=@n14)
   > RemoveOption(self=@n13, Option=Temp)
   ClearSelection(execute=@n12.else, self:"Searchable Combo Box (String) Object Reference"=Get Combo Box())
   n15 = GetOptionAtIndex(self=Get Combo Box(), Index=@n4)
   n16 = Branch(execute=@n4.LoopBody, Condition="Equal Exactly (String)"(A=@n15, B=Get Default Item()))
   Add(execute=@n16.else, TargetArray:Array of Strings=Get Old Items(), NewItem:"String (by ref)"=@n15)
   SetSelectedOption(self=@n13, Option=All)
  ]}
 Update List View {nodes [
   n1 = Update List View!() -> (then, Current Variation:String, Animation Type:String, Filter:String, Current Category:String, Use Cache:bool=~) {is=title K2Node_FunctionEntry}
   > n2 = Sequence!(execute) -> (then_0, then_1, then_2, then_3, then_4:Exec)
   n3 = Get Search AnimSequence Cache()
   n4 = Branch(execute=@n2.then_2, Condition=AND Boolean(A=Get Use Cache() -> (Use Cache:bool), B=Is Not Empty(TargetArray:Array of Strings=@n3)))
   Branch(execute=[@n26.then, @n21.then, @n4.then], Condition=Is Not Empty(TargetArray=Get Search AnimSequence Cache()))
   > n5 = For Each Loop(Array=Get Search AnimSequence Cache())
   > Set L_Current AnimSequence(L_Current AnimSequence:String=@n5) -> (Output_Get:String)
   > n6 = Branch(Condition=IsEmpty(InString=Get Filter() -> (Filter:String)))
   n7 = Branch(execute=[@n12.then, @n6.then], Condition=Get IsCacheSG() -> (IsCacheSG:bool))
   Add Data To List(
    execute=[@n10.then, @n9.else, @n7.then]
    Anim Sequence=To Soft Object Reference(SoftObjectPath_PathString=Get L_Current AnimSequence() -> (L_Current AnimSequence:String))
    Current User=Get Current User()
    Display Subfolder="Equal Exactly (String)"(A=Get Current Animation Type(), B=Get Animation Type Default Item())
   )
   n8 = FindAssetData!(execute=@n7.else, AssetPath:String=GetBaseFilename(InPath=Get L_Current AnimSequence(), bRemovePath=false)) -> (
    then
    ReturnValue_PackageName:Name
    ReturnValue_PackagePath:Name
    ReturnValue_AssetName:Name
    ReturnValue_AssetClassPath:Top Level Asset Path Structure
   )
   > Switch on Name!(
    execute:Exec
    Selection:Name=String To Name(InString=BreakTopLevelAssetPath!(TopLevelAssetPath:"Top Level Asset Path Structure (by ref)"=@n8.ReturnValue_AssetClassPath) -> (PathString:String))
   ) -> (&AnimSequence:Exec) {is=title K2Node_SwitchName, bHasDefaultPin=false, pinNames [&AnimSequence]}
   > n9 = Branch(Condition=Get Filter Match Asset Name(self=Get Save Game()))
   > n10 = Branch(Condition=StartsWith(SourceString="To String (Name)"(InName=@n8.ReturnValue_AssetName), InPrefix=Get Current Asset()))
   Set L_Search String Contains(execute=@n6.else, L_Search String Contains:bool=true) -> (Output_Get:bool)
   > n11 = For Each Loop(Array=Get L_Search String Split() -> (L_Search String Split:Array of Strings))
   n12 = Branch(execute=@n11.Completed, Condition=Get L_Search String Contains() -> (L_Search String Contains:bool))
   n13 = Branch(execute=@n11.LoopBody, Condition=Contains(SearchIn=Get L_Current AnimSequence(), Substring=@n11))
   Set L_Search String Contains(execute=@n13.else, L_Search String Contains=false)
   Set Active AC File(
    execute=@n4.else
    Active AC File:Animatable Creature Object Reference=AnimatableCreature Find(AnimatableCreatures=Get AnimatableCreature(), Asset=Get Current Category() -> (Current Category:String), Variant=Get Current Variation())
   ) -> (Output_Get:Animatable Creature Object Reference)
   > n14 = Branch(Condition=Get isShotgridEnabled())
   > n15 = Cast To AnimatableCreature_ShotGrid(Object=Get Active AC File())
   > n16 = Branch(Condition=Get SG Enabled(self=@n15))
   n17 = Get Plugin_Combobox()
   > n18 = Branch(Condition="Not Equal Exactly (String)"(A=GetSelectedOption(self=@n17), B=All))
   > n19 = Shotgrid - Populate List from Context!(execute, self) -> (then, Success:bool) {is=title K2Node_CallFunction}
   > n20 = Set IsCacheSG(IsCacheSG:bool=@n19) -> (Output_Get:bool)
   > n21 = Branch(Condition=@n20)
   n22 = Get Animation Directories for Asset!(
    execute=[@n21.else, @n27.then]
    self
    Asset Name:String=Get Current Asset()
    Variation:String=Get Current Variation()
    Animation Type:String=Get Current Animation Type()
   ) -> (then, Animation Directories:Array of Names) {is=title K2Node_CallFunction}
   > n23 = GetAssetsByPaths!(
    execute
    self:Asset Registry Interface=GetAssetRegistry()
    PackagePaths:Array of Names=@n22
    bRecursive:bool=Get Search Source Path Recursively!(self:Animatable Creature Object Reference=Get Active AC File()) -> (Search Source Path Recursively:bool)
    bIncludeOnlyOnDiskAssets:bool=false
   ) -> (then, OutAssetData:Array of Asset Data Structures, ReturnValue:bool)
   > n24 = For Each Loop!(Exec, Array:Array of Asset Data Structures=@n23) -> (
    LoopBody
    Array Element_PackageName:Name
    Array Element_PackagePath:Name
    Array Element_AssetName:Name
    Array Element_AssetClassPath:Top Level Asset Path Structure
    Array Index
    Completed
   )
   n25 = Execute Python Script!(
    execute=@n24.Completed
    PythonScript="import os\r\nsorted_files = sorted(files_to_sort, key=lambda x:os.path.basename(x))"
    files_to_sort:Array of Strings=Get Search AnimSequence Cache()
   ) -> (then, ReturnValue, sorted_files:Array of Strings) {inputs [files_to_sort], outputs [sorted_files]}
   > n26 = Set Search AnimSequence Cache(Search AnimSequence Cache:Array of Strings=@n25.sorted_files) -> (Output_Get:Array of Strings)
   Add(
    execute=@n24.LoopBody
    TargetArray=Get Search AnimSequence Cache()
    NewItem=Append!(A="To String (Name)"(InName=@n24), B=., C:String="To String (Name)"(InName=@n24.Array Element_AssetName)) -> (ReturnValue)
   )
   n27 = Set IsCacheSG(execute=[@n18.else, @n16.else, @n15.CastFailed, @n14.else], IsCacheSG=false)
   n28 = Get ListView()
   Set Text(
    execute=@n2.then_4
    Text="To Text (Integer)"(Value=GetNumItems!(self:List View Object Reference=@n28) -> (ReturnValue:int=0))
    self=Get NumFilesTextBlock() -> (NumFilesTextBlock:Text Object Reference)
   )
   n29 = GetListItems!(self:List View Object Reference=Get ListView()) -> (ReturnValue:Array of Object References)
   n30 = Get Save Game()
   n31 = Get Selected Index!(self:SG Animation Browser Object Reference=@n30) -> (Selected Index:int=0)
   n32 = Branch(execute=@n2.then_3, Condition=Is Valid Index!(TargetArray:Array of Object References=@n29, IndexToTest:int=@n31) -> (ReturnValue:bool))
   n33 = "Get (a copy)"(Array:Array of Object References=@n29, Dimension 1=@n31) -> (Output:Object Reference)
   > n34 = Cast To Object_AnimationBrowser(Object=@n33)
   n35 = Get(self=@n34)
   n36 = Get Selected Path!(self:SG Animation Browser Object Reference=@n30) -> (Selected Path:String)
   > n37 = Branch(Condition="Equal Exactly (String)"(A=To Soft Object Path(SoftObjectReference=@n35), B=@n36))
   n38 = Get ListView()
   > Set Selected Item(self=@n38, Item=@n33) {is=title K2Node_CallFunction}
   > Scroll Item Into View!(execute, self:List View Object Reference=@n38, Item:Object Reference=@n33) -> (then)
   n39 = For Each Loop with Break(Exec=[@n37.else, @n32.else], Array:Array of Object References=GetListItems(self=Get ListView()), Break=@n43.then) -> (Array Element:Object Reference)
   > n40 = Cast To Object_AnimationBrowser(Object=@n39)
   n41 = Get(self=@n40)
   > Branch(Condition="Equal, Case Insensitive (String)"(A=To Soft Object Path(SoftObjectReference=@n41), B=@n36))
   n42 = Get ListView()
   > Set Selected Item(self=@n42, Item=@n40)
   > n43 = Scroll Item Into View(self=@n42, Item=@n40)
   ClearListItems(execute=@n2.then_0, self=Get ListView())
   > Set Current Variation(Current Variation:String=@n1) -> (Output_Get:String)
   > Set Current Animation Type(Current Animation Type=@n1.Animation Type)
   > Clear(TargetArray=Get L_Search String Split())
   > Set Current Asset(Current Asset=@n1.Current Category)
   > n44 = Branch(Condition=Get Use Cache())
   Clear(execute=@n44.else, TargetArray=Get Search AnimSequence Cache())
   n45 = For Each Loop!(Exec=@n2.then_1, Array:Array of Strings=ParseIntoArray(SourceString=Get Filter(), Delimiter=" ")) -> (LoopBody, Array Element:String, Array Index, Completed)
   > Append Array(TargetArray:Array of Strings=Get L_Search String Split(), SourceArray:Array of Strings=ParseIntoArray(SourceString=@n45, Delimiter=_))
   Get Current Asset()
   Get Current Variation()
  ]}
 Update List View From UI {nodes [
   Update List View From UI() {is=title K2Node_FunctionEntry}
   n1 = Get SearchTextBox()
   > Update List View(
    Current Variation=GetSelectedOption(self=Get Plugin_Combobox())
    Animation Type=GetSelectedOption(self=Get Subfolder_Combobox())
    Filter="To String (Text)"(InText=Get Text(self:Text Box Object Reference=@n1))
    Current Category=GetSelectedOption(self=Get Asset_ComboBox())
    Use Cache=false
   ) {is=title K2Node_CallFunction}
  ]}
 Update Subfolder List {nodes [
   Update Subfolder List() {is=title K2Node_FunctionEntry}
   n1 = GetSelectedOption(self=Get Subfolder_Combobox())
   > n2 = Branch(Condition="Equal Exactly (String)"(A=@n1, B=All))
   > Return Node!(execute, Updated:bool=false)
   n3 = Branch(execute=@n2.else, Condition=IsEmpty(InString=@n1))
   > RemoveOption(self=Get Subfolder_Combobox(), Option=None)
   > SetSelectedIndex(self=Get Subfolder_Combobox(), Index=0)
   > Return Node(Updated=false)
   n4 = AnimatableCreature Find!(AnimatableCreatures=Get AnimatableCreature(), Asset=Get Current Asset(), Variant=Get Current Variation(), __WorldContext:Object Reference) -> (AnimatableCreature, Exact Match)
   n5 = Cast To AnimatableCreature_ShotGrid(execute=@n3.else, Object=@n4)
   n6 = Get SG Enabled(self=@n5)
   > Set loc_noAllSelection(loc_noAllSelection:bool=AND Boolean(A=Get isShotgridEnabled(), B=@n6)) -> (Output_Get:bool)
   > n7 = Branch(Condition=@n6)
   n8 = Get Subfolders()
   n9 = NOT Boolean(A=Is ComboBox All(CB=Get Plugin_Combobox()) {is=title K2Node_CallFunction})
   n10 = Get loc_noAllSelection() -> (loc_noAllSelection:bool)
   n11 = Get Subfolder_Combobox()
   Update Combo Box(
    execute=[@n14.else, @n7.then]
    Combo Box=@n11
    Items=@n8
    Default Item=SelectString(B=Get Animation Type Default Item(), bPickA=AND Boolean(A=@n10, B=@n9))
    Add Blank=false
   ) {is=title K2Node_CallFunction}
   > SetSelectedIndex(self=@n11, Index=0)
   > Return Node(Updated=true)
   n12 = Get Subfolder_Combobox()
   n13 = Get Subfolders!(self:Animatable Creature Object Reference=@n4) -> (Subfolders)
   n14 = Branch(execute=[@n7.else, @n5.CastFailed], Condition=Contains Item(TargetArray=@n13, ItemToFind=GetSelectedOption(self=@n12)))
   > SetSelectedOption(self=Get Subfolder_Combobox(), Option="Get (a copy)"(Array:Array of Strings=@n13) -> (Output:String))
   > Return Node(Updated=false)
   "Get (a copy)"(Array=@n8)
   Get Subfolder_Combobox()
  ]}
]
variables [
 Active AC File {type=AnimatableCreature_C, default=null}
 Active Project Options {type="array<EU_AnimationBrowser_ProjectOptions_C>", default []}
 AnimatableCreature {type="array<AnimatableCreature_C>", default []}
 Animation Type Default Item {type=string, default=All}
 AnimDialog Anim Sequence {type=Object, default=null}
 AnimDialog Level Sequence {type=Object, default=null}
 Author Tag {type=string, default=AnimationBrowser_Author}
 Current Animation Type {type=string, default=""}
 Current Asset {type=string, default=""}
 Current Filter {type=string, default=""}
 Current Level Sequence {type=LevelSequence, default=null}
 Current List Item {type=AnimationBrowser_ListItem_C, default=null}
 Current Project {type=BP_AnimationBrowser_Project_C, default=null}
 Current User {type=string, default=""}
 Current Variation {type=string, default=""}
 Default Map Tag {type=string, default=DefaultMap}
 Developer Mode {type=boolean}
 Is Setup {type=boolean}
 IsCacheSG {type=boolean}
 isShotgridEnabled {type=boolean}
 Level {type=World, default=null}
 Save Game {type=SG_AnimationBrowser_C, default=null}
 Search AnimSequence Cache {type="array<string>", default []}
 Subfolders From Disk Cache {type="array<string>", default []}
]
widgets {
 info {rootWidgetClass=&CanvasPanel, widgetCount=51}
 widgets [
  CanvasPanel {widgetClassPath=&CanvasPanel, bIsVariable=true, widget=CanvasPanel}
  VerticalBox_0 {
   widgetClassPath=&VerticalBox
   parent=CanvasPanel
   widget=VerticalBox_0
   slot {class=CanvasPanelSlot, properties {layoutData {alignment=<0 0>, anchors {maximum=<1 1>, minimum=<0 0>}, offsets=<ltrb 0 0 0 0>}}}
  }
  HeaderBorder {
   widgetClassPath=&Border
   parent=VerticalBox_0
   widget=HeaderBorder
   properties {brushColor=<rgba 0.0275 0.0275 0.0275 1>, padding=<ltrb 2 2 2 2>}
   slot {class=VerticalBoxSlot, properties {padding=<ltrb 0 0 0 1>}}
  }
  VerticalBox_1 {
   widgetClassPath=&VerticalBox
   parent=HeaderBorder
   widget=VerticalBox_1
   slot {class=BorderSlot, properties {padding=<ltrb 2 2 2 2>}}
  }
  HorizontalBox_1 {widgetClassPath=&HorizontalBox, parent=VerticalBox_1, widget=HorizontalBox_1, slot {class=VerticalBoxSlot}}
  AssetImage {
   widgetClassPath=&Image
   bIsVariable=true
   parent=HorizontalBox_1
   widget=AssetImage
   properties {brush {
     bIsDynamicallyLoaded=false
     drawAs=Image
     imageType=FullColor
     mirroring=NoMirror
     resourceName=None
     resourceObject=/AnimInEdTools/AnimationBrowser/Icons/Icon_Asset
     tiling=NoTile
     imageSize=<18 28>
     margin=<ltrb 0 0 0 0>
     outlineSettings {
      bUseBrushTransparency=false
      roundingType=HalfHeightRadius
      width=0
      color {colorUseRule=UseColor_Specified, specifiedColor=<rgba 0 0 0 0>}
      cornerRadii=<0 0 0 0>
     }
     tintColor {colorUseRule=UseColor_Specified, specifiedColor=<rgba 1 1 1 1>}
     uVRegion {bIsValid=false, max=<0 0>, min=<0 0>}
    }}
   slot {class=HorizontalBoxSlot, properties {horizontalAlignment=HAlign_Center, verticalAlignment=VAlign_Center}}
  }
  SizeBox_3 {
   widgetClassPath=&SizeBox
   parent=HorizontalBox_1
   widget=SizeBox_3
   properties {
    bOverride_MaxDesiredHeight=true
    bOverride_MaxDesiredWidth=true
    bOverride_MinDesiredWidth=true
    maxDesiredHeight=25
    maxDesiredWidth=200
    minDesiredWidth=100
   }
   slot {class=HorizontalBoxSlot, properties {size {sizeRule=Fill, value=1}}}
  }
  Asset_ComboBox {
   widgetClassPath=/Script/FortniteEditor.SearchableStringComboBox
   bIsVariable=true
   parent=SizeBox_3
   widget=Asset_ComboBox
   properties {
    selectedOption=Category
    toolTipText=Category Dropdown: Corresponds to 'Asset' field in AC Files
    defaultOptions [Category]
    font {
     bForceMonospaced=false
     bMaterialIsStencil=false
     fontMaterial=null
     fontObject=&Roboto
     letterSpacing=0
     monospacedWidth=1
     size=12
     skewAmount=0
     typefaceFontName=Bold
     outlineSettings {
      bApplyOutlineToDropShadows=false
      bMiteredCorners=false
      bSeparateFillAlpha=false
      outlineMaterial=null
      outlineSize=0
      outlineColor=<rgba 0 0 0 1>
     }
    }
    foregroundColor {colorUseRule=UseColor_Specified, specifiedColor=<rgba 0.5271 0.5271 0.5271 1>}
    itemStyle {
     bUseParentRowBrush=false
     activeBrush {
      bIsDynamicallyLoaded=false
      drawAs=Image
      imageType=NoImage
      mirroring=NoMirror
      resourceName=None
      resourceObject=null
      tiling=NoTile
      imageSize=<0 0>
      margin=<ltrb 0 0 0 0>
      outlineSettings {
       bUseBrushTransparency=false
       roundingType=HalfHeightRadius
       width=0
       color {colorUseRule=UseColor_Specified, specifiedColor=<rgba 0 0 0 0>}
       cornerRadii=<0 0 0 0>
      }
      tintColor {colorUseRule=UseColor_Specified, specifiedColor=<rgba 0 0.162 0.7454 1>}
      uVRegion {bIsValid=false, max=<0 0>, min=<0 0>}
     }
     activeHighlightedBrush {
      bIsDynamicallyLoaded=false
      drawAs=Image
      imageType=NoImage
      mirroring=NoMirror
      resourceName=None
      resourceObject=null
      tiling=NoTile
      imageSize=<0 0>
      margin=<ltrb 0 0 0 0>
      outlineSettings {
       bUseBrushTransparency=false
       roundingType=HalfHeightRadius
       width=0
       color {colorUseRule=UseColor_Specified, specifiedColor=<rgba 0 0 0 0>}
       cornerRadii=<0 0 0 0>
      }
      tintColor {colorUseRule=UseColor_Specified, specifiedColor=<rgba 0.0252 0.0319 0.0423 1>}
      uVRegion {bIsValid=false, max=<0 0>, min=<0 0>}
     }
     activeHoveredBrush {
      bIsDynamicallyLoaded=false
      drawAs=Image
      imageType=NoImage
      mirroring=NoMirror
      resourceName=None
      resourceObject=null
      tiling=NoTile
      imageSize=<0 0>
      margin=<ltrb 0 0 0 0>
      outlineSettings {
       bUseBrushTransparency=false
       roundingType=HalfHeightRadius
       width=0
       color {colorUseRule=UseColor_Specified, specifiedColor=<rgba 0 0 0 0>}
       cornerRadii=<0 0 0 0>
      }
      tintColor {colorUseRule=UseColor_Specified, specifiedColor=<rgba 0 0.162 0.7454 1>}
      uVRegion {bIsValid=false, max=<0 0>, min=<0 0>}
     }
     dropIndicator_Above {
      bIsDynamicallyLoaded=false
      drawAs=Box
      imageType=FullColor
      mirroring=NoMirror
      resourceName=../../../Engine/Content/Slate/Common/DropZoneIndicator_Above.png
      resourceObject=null
      tiling=NoTile
      imageSize=<0 0>
      margin=<ltrb 0.625 0.625 0 0>
      outlineSettings {
       bUseBrushTransparency=false
       roundingType=HalfHeightRadius
       width=0
       color {colorUseRule=UseColor_Specified, specifiedColor=<rgba 0 0 0 0>}
       cornerRadii=<0 0 0 0>
      }
      tintColor {colorUseRule=UseColor_Specified, specifiedColor=<rgba 0.0091 0.3515 0.9647 1>}
      uVRegion {bIsValid=false, max=<0 0>, min=<0 0>}
     }
     dropIndicator_Below {
      bIsDynamicallyLoaded=false
      drawAs=Box
      imageType=FullColor
      mirroring=NoMirror
      resourceName=../../../Engine/Content/Slate/Common/DropZoneIndicator_Below.png
      resourceObject=null
      tiling=NoTile
      imageSize=<0 0>
      margin=<ltrb 0.625 0 0 0.625>
      outlineSettings {
       bUseBrushTransparency=false
       roundingType=HalfHeightRadius
       width=0
       color {colorUseRule=UseColor_Specified, specifiedColor=<rgba 0 0 0 0>}
       cornerRadii=<0 0 0 0>
      }
      tintColor {colorUseRule=UseColor_Specified, specifiedColor=<rgba 0.0091 0.3515 0.9647 1>}
      uVRegion {bIsValid=false, max=<0 0>, min=<0 0>}
     }
     dropIndicator_Onto {
      bIsDynamicallyLoaded=false
      drawAs=Box
      imageType=FullColor
      mirroring=NoMirror
      resourceName=../../../Engine/Content/Slate/Common/DropZoneIndicator_Onto.png
      resourceObject=null
      tiling=NoTile
      imageSize=<0 0>
      margin=<ltrb 0.25 0.25 0.25 0.25>
      outlineSettings {
       bUseBrushTransparency=false
       roundingType=HalfHeightRadius
       width=0
       color {colorUseRule=UseColor_Specified, specifiedColor=<rgba 0 0 0 0>}
       cornerRadii=<0 0 0 0>
      }
      tintColor {colorUseRule=UseColor_Specified, specifiedColor=<rgba 0.0091 0.3515 0.9647 1>}
      uVRegion {bIsValid=false, max=<0 0>, min=<0 0>}
     }
     evenRowBackgroundBrush {
      bIsDynamicallyLoaded=false
      drawAs=Image
      imageType=NoImage
      mirroring=NoMirror
      resourceName=None
      resourceObject=null
      tiling=NoTile
      imageSize=<0 0>
      margin=<ltrb 0 0 0 0>
      outlineSettings {
       bUseBrushTransparency=false
       roundingType=HalfHeightRadius
       width=0
       color {colorUseRule=UseColor_Specified, specifiedColor=<rgba 0 0 0 0>}
       cornerRadii=<0 0 0 0>
      }
      tintColor {colorUseRule=UseColor_Specified, specifiedColor=<rgba 0.0103 0.0103 0.0103 1>}
      uVRegion {bIsValid=false, max=<0 0>, min=<0 0>}
     }
     evenRowBackgroundHoveredBrush {
      bIsDynamicallyLoaded=false
      drawAs=Image
      imageType=NoImage
      mirroring=NoMirror
      resourceName=None
      resourceObject=null
      tiling=NoTile
      imageSize=<0 0>
      margin=<ltrb 0 0 0 0>
      outlineSettings {
       bUseBrushTransparency=false
       roundingType=HalfHeightRadius
       width=0
       color {colorUseRule=UseColor_Specified, specifiedColor=<rgba 0 0 0 0>}
       cornerRadii=<0 0 0 0>
      }
      tintColor {colorUseRule=UseColor_Specified, specifiedColor=<rgba 0.0176 0.0176 0.0176 1>}
      uVRegion {bIsValid=false, max=<0 0>, min=<0 0>}
     }
     inactiveBrush {
      bIsDynamicallyLoaded=false
      drawAs=Image
      imageType=NoImage
      mirroring=NoMirror
      resourceName=None
      resourceObject=null
      tiling=NoTile
      imageSize=<0 0>
      margin=<ltrb 0 0 0 0>
      outlineSettings {
       bUseBrushTransparency=false
       roundingType=HalfHeightRadius
       width=0
       color {colorUseRule=UseColor_Specified, specifiedColor=<rgba 0 0 0 0>}
       cornerRadii=<0 0 0 0>
      }
      tintColor {colorUseRule=UseColor_Specified, specifiedColor=<rgba 0.0513 0.0953 0.159 1>}
      uVRegion {bIsValid=false, max=<0 0>, min=<0 0>}
     }
     inactiveHighlightedBrush {
      bIsDynamicallyLoaded=false
      drawAs=Image
      imageType=NoImage
      mirroring=NoMirror
      resourceName=None
      resourceObject=null
      tiling=NoTile
      imageSize=<0 0>
      margin=<ltrb 0 0 0 0>
      outlineSettings {
       bUseBrushTransparency=false
       roundingType=HalfHeightRadius
       width=0
       color {colorUseRule=UseColor_Specified, specifiedColor=<rgba 0 0 0 0>}
       cornerRadii=<0 0 0 0>
      }
      tintColor {colorUseRule=UseColor_Specified, specifiedColor=<rgba 0.0252 0.0319 0.0423 1>}
      uVRegion {bIsValid=false, max=<0 0>, min=<0 0>}
     }
     inactiveHoveredBrush {
      bIsDynamicallyLoaded=false
      drawAs=Image
      imageType=NoImage
      mirroring=NoMirror
      resourceName=None
      resourceObject=null
      tiling=NoTile
      imageSize=<0 0>
      margin=<ltrb 0 0 0 0>
      outlineSettings {
       bUseBrushTransparency=false
       roundingType=HalfHeightRadius
       width=0
       color {colorUseRule=UseColor_Specified, specifiedColor=<rgba 0 0 0 0>}
       cornerRadii=<0 0 0 0>
      }
      tintColor {colorUseRule=UseColor_Specified, specifiedColor=<rgba 0.0513 0.0953 0.159 1>}
      uVRegion {bIsValid=false, max=<0 0>, min=<0 0>}
     }
     oddRowBackgroundBrush {
      bIsDynamicallyLoaded=false
      drawAs=Image
      imageType=NoImage
      mirroring=NoMirror
      resourceName=None
      resourceObject=null
      tiling=NoTile
      imageSize=<0 0>
      margin=<ltrb 0 0 0 0>
      outlineSettings {
       bUseBrushTransparency=false
       roundingType=HalfHeightRadius
       width=0
       color {colorUseRule=UseColor_Specified, specifiedColor=<rgba 0 0 0 0>}
       cornerRadii=<0 0 0 0>
      }
      tintColor {colorUseRule=UseColor_Specified, specifiedColor=<rgba 0.0103 0.0103 0.0103 1>}
      uVRegion {bIsValid=false, max=<0 0>, min=<0 0>}
     }
     oddRowBackgroundHoveredBrush {
      bIsDynamicallyLoaded=false
      drawAs=Image
      imageType=NoImage
      mirroring=NoMirror
      resourceName=None
      resourceObject=null
      tiling=NoTile
      imageSize=<0 0>
      margin=<ltrb 0 0 0 0>
      outlineSettings {
       bUseBrushTransparency=false
       roundingType=HalfHeightRadius
       width=0
       color {colorUseRule=UseColor_Specified, specifiedColor=<rgba 0 0 0 0>}
       cornerRadii=<0 0 0 0>
      }
      tintColor {colorUseRule=UseColor_Specified, specifiedColor=<rgba 0.0176 0.0176 0.0176 1>}
      uVRegion {bIsValid=false, max=<0 0>, min=<0 0>}
     }
     parentRowBackgroundBrush {
      bIsDynamicallyLoaded=false
      drawAs=Image
      imageType=NoImage
      mirroring=NoMirror
      resourceName=None
      resourceObject=null
      tiling=NoTile
      imageSize=<32 32>
      margin=<ltrb 0 0 0 0>
      outlineSettings {
       bUseBrushTransparency=false
       roundingType=HalfHeightRadius
       width=0
       color {colorUseRule=UseColor_Specified, specifiedColor=<rgba 0 0 0 0>}
       cornerRadii=<0 0 0 0>
      }
      tintColor {colorUseRule=UseColor_Specified, specifiedColor=<rgba 1 1 1 1>}
      uVRegion {bIsValid=false, max=<0 0>, min=<0 0>}
     }
     parentRowBackgroundHoveredBrush {
      bIsDynamicallyLoaded=false
      drawAs=Image
      imageType=NoImage
      mirroring=NoMirror
      resourceName=None
      resourceObject=null
      tiling=NoTile
      imageSize=<32 32>
      margin=<ltrb 0 0 0 0>
      outlineSettings {
       bUseBrushTransparency=false
       roundingType=HalfHeightRadius
       width=0
       color {colorUseRule=UseColor_Specified, specifiedColor=<rgba 0 0 0 0>}
       cornerRadii=<0 0 0 0>
      }
      tintColor {colorUseRule=UseColor_Specified, specifiedColor=<rgba 1 1 1 1>}
      uVRegion {bIsValid=false, max=<0 0>, min=<0 0>}
     }
     selectedTextColor {colorUseRule=UseColor_Specified, specifiedColor=<rgba 1 1 1 1>}
     selectorFocusedBrush {
      bIsDynamicallyLoaded=false
      drawAs=Border
      imageType=FullColor
      mirroring=NoMirror
      resourceName=../../../Engine/Content/Slate/Common/Selector.png
      resourceObject=null
      tiling=Both
      imageSize=<0 0>
      margin=<ltrb 0.25 0.25 0.25 0.25>
      outlineSettings {
       bUseBrushTransparency=false
       roundingType=HalfHeightRadius
       width=0
       color {colorUseRule=UseColor_Specified, specifiedColor=<rgba 0 0 0 0>}
       cornerRadii=<0 0 0 0>
      }
      tintColor {colorUseRule=UseColor_Specified, specifiedColor=<rgba 0.701 0.225 0.003 1>}
      uVRegion {bIsValid=false, max=<0 0>, min=<0 0>}
     }
     textColor {colorUseRule=UseColor_Specified, specifiedColor=<rgba 0.5271 0.5271 0.5271 1>}
    }
    widgetStyle {
     comboButtonStyle {
      downArrowAlign=VAlign_Center
      buttonStyle {
       clickedSlateSound {resourceObject=null}
       disabled {
        bIsDynamicallyLoaded=false
        drawAs=RoundedBox
        imageType=NoImage
        mirroring=NoMirror
        resourceName=None
        resourceObject=null
        tiling=NoTile
        imageSize=<0 0>
        margin=<ltrb 0 0 0 0>
        outlineSettings {
         bUseBrushTransparency=false
         roundingType=FixedRadius
         width=1
         color {colorUseRule=UseColor_ColorTable, specifiedColor=<rgba 0 0 0 0>}
         cornerRadii=<4 4 4 4>
        }
        tintColor {colorUseRule=UseColor_Specified, specifiedColor=<rgba 0.0395 0.0395 0.0395 1>}
        uVRegion {bIsValid=false, max=<0 0>, min=<0 0>}
       }
       disabledForeground {colorUseRule=UseColor_ColorTable, specifiedColor=<rgba 0 0 0 0>}
       hovered {
        bIsDynamicallyLoaded=false
        drawAs=RoundedBox
        imageType=NoImage
        mirroring=NoMirror
        resourceName=None
        resourceObject=null
        tiling=NoTile
        imageSize=<0 0>
        margin=<ltrb 0 0 0 0>
        outlineSettings {
         bUseBrushTransparency=false
         roundingType=FixedRadius
         width=1
         color {colorUseRule=UseColor_ColorTable, specifiedColor=<rgba 0 0 0 0>}
         cornerRadii=<4 4 4 4>
        }
        tintColor {colorUseRule=UseColor_Specified, specifiedColor=<rgba 0.0048 0.0048 0.0048 1>}
        uVRegion {bIsValid=false, max=<0 0>, min=<0 0>}
       }
       hoveredForeground {colorUseRule=UseColor_ColorTable, specifiedColor=<rgba 0 0 0 0>}
       hoveredSlateSound {resourceObject=null}
       normal {
        bIsDynamicallyLoaded=false
        drawAs=RoundedBox
        imageType=NoImage
        mirroring=NoMirror
        resourceName=None
        resourceObject=null
        tiling=NoTile
        imageSize=<0 0>
        margin=<ltrb 0 0 0 0>
        outlineSettings {
         bUseBrushTransparency=false
         roundingType=FixedRadius
         width=1
         color {colorUseRule=UseColor_ColorTable, specifiedColor=<rgba 0 0 0 0>}
         cornerRadii=<4 4 4 4>
        }
        tintColor {colorUseRule=UseColor_Specified, specifiedColor=<rgba 0.0048 0.0048 0.0048 1>}
        uVRegion {bIsValid=false, max=<0 0>, min=<0 0>}
       }
       normalForeground {colorUseRule=UseColor_ColorTable, specifiedColor=<rgba 0 0 0 0>}
       normalPadding=<ltrb 8 1 8 1>
       pressed {
        bIsDynamicallyLoaded=false
        drawAs=RoundedBox
        imageType=NoImage
        mirroring=NoMirror
        resourceName=None
        resourceObject=null
        tiling=NoTile
        imageSize=<0 0>
        margin=<ltrb 0 0 0 0>
        outlineSettings {
         bUseBrushTransparency=false
         roundingType=FixedRadius
         width=1
         color {colorUseRule=UseColor_ColorTable, specifiedColor=<rgba 0 0 0 0>}
         cornerRadii=<4 4 4 4>
        }
        tintColor {colorUseRule=UseColor_Specified, specifiedColor=<rgba 0.0048 0.0048 0.0048 1>}
        uVRegion {bIsValid=false, max=<0 0>, min=<0 0>}
       }
       pressedForeground {colorUseRule=UseColor_ColorTable, specifiedColor=<rgba 0 0 0 0>}
       pressedPadding=<ltrb 8 2 8 0>
       pressedSlateSound {resourceObject=null}
      }
      contentPadding=<ltrb 0 0 0 0>
      downArrowImage {
       bIsDynamicallyLoaded=false
       drawAs=Image
       imageType=Vector
       mirroring=NoMirror
       resourceName=../../../Engine/Content/Slate/Starship/CoreWidgets/ComboBox/wide-chevron-down.svg
       resourceObject=null
       tiling=NoTile
       imageSize=<16 16>
       margin=<ltrb 0 0 0 0>
       outlineSettings {
        bUseBrushTransparency=false
        roundingType=HalfHeightRadius
        width=0
        color {colorUseRule=UseColor_Specified, specifiedColor=<rgba 0 0 0 0>}
        cornerRadii=<0 0 0 0>
       }
       tintColor {colorUseRule=UseColor_Specified, specifiedColor=<rgba 1 1 1 1>}
       uVRegion {bIsValid=false, max=<0 0>, min=<0 0>}
      }
      downArrowPadding=<ltrb 2 2 2 2>
      menuBorderBrush {
       bIsDynamicallyLoaded=false
       drawAs=RoundedBox
       imageType=NoImage
       mirroring=NoMirror
       resourceName=None
       resourceObject=null
       tiling=NoTile
       imageSize=<0 0>
       margin=<ltrb 0 0 0 0>
       outlineSettings {
        bUseBrushTransparency=false
        roundingType=FixedRadius
        width=1
        color {colorUseRule=UseColor_ColorTable, specifiedColor=<rgba 0 0 0 0>}
        cornerRadii=<0 0 0 0>
       }
       tintColor {colorUseRule=UseColor_Specified, specifiedColor=<rgba 0.0395 0.0395 0.0395 1>}
       uVRegion {bIsValid=false, max=<0 0>, min=<0 0>}
      }
      menuBorderPadding=<ltrb 0 0 0 0>
      shadowColorAndOpacity=<rgba 0 0 0 1>
      shadowOffset=<0 0>
     }
     contentPadding=<ltrb 0 0 0 0>
     menuRowPadding=<ltrb 8 3 8 3>
     pressedSlateSound {resourceObject=null}
     selectionChangeSlateSound {resourceObject=null}
    }
   }
   slot {class=SizeBoxSlot}
  }
  PluginHorizontalBox {
   widgetClassPath=&HorizontalBox
   bIsVariable=true
   parent=HorizontalBox_1
   widget=PluginHorizontalBox
   slot {class=HorizontalBoxSlot, properties {horizontalAlignment=HAlign_Center, verticalAlignment=VAlign_Center}}
  }
  PluginImage {
   widgetClassPath=&Image
   bIsVariable=true
   parent=PluginHorizontalBox
   widget=PluginImage
   properties {brush {
     bIsDynamicallyLoaded=false
     drawAs=Image
     imageType=FullColor
     mirroring=NoMirror
     resourceName=None
     resourceObject=/AnimInEdTools/AnimationBrowser/Icons/Icon_Plugin
     tiling=NoTile
     imageSize=<18 28>
     margin=<ltrb 0 0 0 0>
     outlineSettings {
      bUseBrushTransparency=false
      roundingType=HalfHeightRadius
      width=0
      color {colorUseRule=UseColor_Specified, specifiedColor=<rgba 0 0 0 0>}
      cornerRadii=<0 0 0 0>
     }
     tintColor {colorUseRule=UseColor_Specified, specifiedColor=<rgba 1 1 1 1>}
     uVRegion {bIsValid=false, max=<0 0>, min=<0 0>}
    }}
   slot {class=HorizontalBoxSlot}
  }
  SizeBox_1 {
   widgetClassPath=&SizeBox
   parent=PluginHorizontalBox
   widget=SizeBox_1
   properties {
    bOverride_MaxDesiredHeight=true
    bOverride_MaxDesiredWidth=true
    bOverride_MinDesiredWidth=true
    maxDesiredHeight=25
    maxDesiredWidth=200
    minDesiredWidth=100
   }
   slot {class=HorizontalBoxSlot}
  }
  Plugin_Combobox {
   widgetClassPath=&SearchableStringComboBox
   bIsVariable=true
   parent=SizeBox_1
   widget=Plugin_Combobox
   properties {
    selectedOption=Select Character
    toolTipText=Variant Dropdown: Corresponds to 'Variant' field in AC Files
    defaultOptions [Select Character]
    font {
     bForceMonospaced=false
     bMaterialIsStencil=false
     fontMaterial=null
     fontObject=&Roboto
     letterSpacing=0
     monospacedWidth=1
     size=12
     skewAmount=0
     typefaceFontName=Bold
     outlineSettings {
      bApplyOutlineToDropShadows=false
      bMiteredCorners=false
      bSeparateFillAlpha=false
      outlineMaterial=null
      outlineSize=0
      outlineColor=<rgba 0 0 0 1>
     }
    }
    foregroundColor {colorUseRule=UseColor_Specified, specifiedColor=<rgba 0.5271 0.5271 0.5271 1>}
    itemStyle {
     bUseParentRowBrush=false
     activeBrush {
      bIsDynamicallyLoaded=false
      drawAs=Image
      imageType=NoImage
      mirroring=NoMirror
      resourceName=None
      resourceObject=null
      tiling=NoTile
      imageSize=<0 0>
      margin=<ltrb 0 0 0 0>
      outlineSettings {
       bUseBrushTransparency=false
       roundingType=HalfHeightRadius
       width=0
       color {colorUseRule=UseColor_Specified, specifiedColor=<rgba 0 0 0 0>}
       cornerRadii=<0 0 0 0>
      }
      tintColor {colorUseRule=UseColor_Specified, specifiedColor=<rgba 0 0.162 0.7454 1>}
      uVRegion {bIsValid=false, max=<0 0>, min=<0 0>}
     }
     activeHighlightedBrush {
      bIsDynamicallyLoaded=false
      drawAs=Image
      imageType=NoImage
      mirroring=NoMirror
      resourceName=None
      resourceObject=null
      tiling=NoTile
      imageSize=<0 0>
      margin=<ltrb 0 0 0 0>
      outlineSettings {
       bUseBrushTransparency=false
       roundingType=HalfHeightRadius
       width=0
       color {colorUseRule=UseColor_Specified, specifiedColor=<rgba 0 0 0 0>}
       cornerRadii=<0 0 0 0>
      }
      tintColor {colorUseRule=UseColor_Specified, specifiedColor=<rgba 0.0252 0.0319 0.0423 1>}
      uVRegion {bIsValid=false, max=<0 0>, min=<0 0>}
     }
     activeHoveredBrush {
      bIsDynamicallyLoaded=false
      drawAs=Image
      imageType=NoImage
      mirroring=NoMirror
      resourceName=None
      resourceObject=null
      tiling=NoTile
      imageSize=<0 0>
      margin=<ltrb 0 0 0 0>
      outlineSettings {
       bUseBrushTransparency=false
       roundingType=HalfHeightRadius
       width=0
       color {colorUseRule=UseColor_Specified, specifiedColor=<rgba 0 0 0 0>}
       cornerRadii=<0 0 0 0>
      }
      tintColor {colorUseRule=UseColor_Specified, specifiedColor=<rgba 0 0.162 0.7454 1>}
      uVRegion {bIsValid=false, max=<0 0>, min=<0 0>}
     }
     dropIndicator_Above {
      bIsDynamicallyLoaded=false
      drawAs=Box
      imageType=FullColor
      mirroring=NoMirror
      resourceName=../../../Engine/Content/Slate/Common/DropZoneIndicator_Above.png
      resourceObject=null
      tiling=NoTile
      imageSize=<0 0>
      margin=<ltrb 0.625 0.625 0 0>
      outlineSettings {
       bUseBrushTransparency=false
       roundingType=HalfHeightRadius
       width=0
       color {colorUseRule=UseColor_Specified, specifiedColor=<rgba 0 0 0 0>}
       cornerRadii=<0 0 0 0>
      }
      tintColor {colorUseRule=UseColor_Specified, specifiedColor=<rgba 0.0091 0.3515 0.9647 1>}
      uVRegion {bIsValid=false, max=<0 0>, min=<0 0>}
     }
     dropIndicator_Below {
      bIsDynamicallyLoaded=false
      drawAs=Box
      imageType=FullColor
      mirroring=NoMirror
      resourceName=../../../Engine/Content/Slate/Common/DropZoneIndicator_Below.png
      resourceObject=null
      tiling=NoTile
      imageSize=<0 0>
      margin=<ltrb 0.625 0 0 0.625>
      outlineSettings {
       bUseBrushTransparency=false
       roundingType=HalfHeightRadius
       width=0
       color {colorUseRule=UseColor_Specified, specifiedColor=<rgba 0 0 0 0>}
       cornerRadii=<0 0 0 0>
      }
      tintColor {colorUseRule=UseColor_Specified, specifiedColor=<rgba 0.0091 0.3515 0.9647 1>}
      uVRegion {bIsValid=false, max=<0 0>, min=<0 0>}
     }
     dropIndicator_Onto {
      bIsDynamicallyLoaded=false
      drawAs=Box
      imageType=FullColor
      mirroring=NoMirror
      resourceName=../../../Engine/Content/Slate/Common/DropZoneIndicator_Onto.png
      resourceObject=null
      tiling=NoTile
      imageSize=<0 0>
      margin=<ltrb 0.25 0.25 0.25 0.25>
      outlineSettings {
       bUseBrushTransparency=false
       roundingType=HalfHeightRadius
       width=0
       color {colorUseRule=UseColor_Specified, specifiedColor=<rgba 0 0 0 0>}
       cornerRadii=<0 0 0 0>
      }
      tintColor {colorUseRule=UseColor_Specified, specifiedColor=<rgba 0.0091 0.3515 0.9647 1>}
      uVRegion {bIsValid=false, max=<0 0>, min=<0 0>}
     }
     evenRowBackgroundBrush {
      bIsDynamicallyLoaded=false
      drawAs=Image
      imageType=NoImage
      mirroring=NoMirror
      resourceName=None
      resourceObject=null
      tiling=NoTile
      imageSize=<0 0>
      margin=<ltrb 0 0 0 0>
      outlineSettings {
       bUseBrushTransparency=false
       roundingType=HalfHeightRadius
       width=0
       color {colorUseRule=UseColor_Specified, specifiedColor=<rgba 0 0 0 0>}
       cornerRadii=<0 0 0 0>
      }
      tintColor {colorUseRule=UseColor_Specified, specifiedColor=<rgba 0.0103 0.0103 0.0103 1>}
      uVRegion {bIsValid=false, max=<0 0>, min=<0 0>}
     }
     evenRowBackgroundHoveredBrush {
      bIsDynamicallyLoaded=false
      drawAs=Image
      imageType=NoImage
      mirroring=NoMirror
      resourceName=None
      resourceObject=null
      tiling=NoTile
      imageSize=<0 0>
      margin=<ltrb 0 0 0 0>
      outlineSettings {
       bUseBrushTransparency=false
       roundingType=HalfHeightRadius
       width=0
       color {colorUseRule=UseColor_Specified, specifiedColor=<rgba 0 0 0 0>}
       cornerRadii=<0 0 0 0>
      }
      tintColor {colorUseRule=UseColor_Specified, specifiedColor=<rgba 0.0176 0.0176 0.0176 1>}
      uVRegion {bIsValid=false, max=<0 0>, min=<0 0>}
     }
     inactiveBrush {
      bIsDynamicallyLoaded=false
      drawAs=Image
      imageType=NoImage
      mirroring=NoMirror
      resourceName=None
      resourceObject=null
      tiling=NoTile
      imageSize=<0 0>
      margin=<ltrb 0 0 0 0>
      outlineSettings {
       bUseBrushTransparency=false
       roundingType=HalfHeightRadius
       width=0
       color {colorUseRule=UseColor_Specified, specifiedColor=<rgba 0 0 0 0>}
       cornerRadii=<0 0 0 0>
      }
      tintColor {colorUseRule=UseColor_Specified, specifiedColor=<rgba 0.0513 0.0953 0.159 1>}
      uVRegion {bIsValid=false, max=<0 0>, min=<0 0>}
     }
     inactiveHighlightedBrush {
      bIsDynamicallyLoaded=false
      drawAs=Image
      imageType=NoImage
      mirroring=NoMirror
      resourceName=None
      resourceObject=null
      tiling=NoTile
      imageSize=<0 0>
      margin=<ltrb 0 0 0 0>
      outlineSettings {
       bUseBrushTransparency=false
       roundingType=HalfHeightRadius
       width=0
       color {colorUseRule=UseColor_Specified, specifiedColor=<rgba 0 0 0 0>}
       cornerRadii=<0 0 0 0>
      }
      tintColor {colorUseRule=UseColor_Specified, specifiedColor=<rgba 0.0252 0.0319 0.0423 1>}
      uVRegion {bIsValid=false, max=<0 0>, min=<0 0>}
     }
     inactiveHoveredBrush {
      bIsDynamicallyLoaded=false
      drawAs=Image
      imageType=NoImage
      mirroring=NoMirror
      resourceName=None
      resourceObject=null
      tiling=NoTile
      imageSize=<0 0>
      margin=<ltrb 0 0 0 0>
      outlineSettings {
       bUseBrushTransparency=false
       roundingType=HalfHeightRadius
       width=0
       color {colorUseRule=UseColor_Specified, specifiedColor=<rgba 0 0 0 0>}
       cornerRadii=<0 0 0 0>
      }
      tintColor {colorUseRule=UseColor_Specified, specifiedColor=<rgba 0.0513 0.0953 0.159 1>}
      uVRegion {bIsValid=false, max=<0 0>, min=<0 0>}
     }
     oddRowBackgroundBrush {
      bIsDynamicallyLoaded=false
      drawAs=Image
      imageType=NoImage
      mirroring=NoMirror
      resourceName=None
      resourceObject=null
      tiling=NoTile
      imageSize=<0 0>
      margin=<ltrb 0 0 0 0>
      outlineSettings {
       bUseBrushTransparency=false
       roundingType=HalfHeightRadius
       width=0
       color {colorUseRule=UseColor_Specified, specifiedColor=<rgba 0 0 0 0>}
       cornerRadii=<0 0 0 0>
      }
      tintColor {colorUseRule=UseColor_Specified, specifiedColor=<rgba 0.0103 0.0103 0.0103 1>}
      uVRegion {bIsValid=false, max=<0 0>, min=<0 0>}
     }
     oddRowBackgroundHoveredBrush {
      bIsDynamicallyLoaded=false
      drawAs=Image
      imageType=NoImage
      mirroring=NoMirror
      resourceName=None
      resourceObject=null
      tiling=NoTile
      imageSize=<0 0>
      margin=<ltrb 0 0 0 0>
      outlineSettings {
       bUseBrushTransparency=false
       roundingType=HalfHeightRadius
       width=0
       color {colorUseRule=UseColor_Specified, specifiedColor=<rgba 0 0 0 0>}
       cornerRadii=<0 0 0 0>
      }
      tintColor {colorUseRule=UseColor_Specified, specifiedColor=<rgba 0.0176 0.0176 0.0176 1>}
      uVRegion {bIsValid=false, max=<0 0>, min=<0 0>}
     }
     parentRowBackgroundBrush {
      bIsDynamicallyLoaded=false
      drawAs=Image
      imageType=NoImage
      mirroring=NoMirror
      resourceName=None
      resourceObject=null
      tiling=NoTile
      imageSize=<32 32>
      margin=<ltrb 0 0 0 0>
      outlineSettings {
       bUseBrushTransparency=false
       roundingType=HalfHeightRadius
       width=0
       color {colorUseRule=UseColor_Specified, specifiedColor=<rgba 0 0 0 0>}
       cornerRadii=<0 0 0 0>
      }
      tintColor {colorUseRule=UseColor_Specified, specifiedColor=<rgba 1 1 1 1>}
      uVRegion {bIsValid=false, max=<0 0>, min=<0 0>}
     }
     parentRowBackgroundHoveredBrush {
      bIsDynamicallyLoaded=false
      drawAs=Image
      imageType=NoImage
      mirroring=NoMirror
      resourceName=None
      resourceObject=null
      tiling=NoTile
      imageSize=<32 32>
      margin=<ltrb 0 0 0 0>
      outlineSettings {
       bUseBrushTransparency=false
       roundingType=HalfHeightRadius
       width=0
       color {colorUseRule=UseColor_Specified, specifiedColor=<rgba 0 0 0 0>}
       cornerRadii=<0 0 0 0>
      }
      tintColor {colorUseRule=UseColor_Specified, specifiedColor=<rgba 1 1 1 1>}
      uVRegion {bIsValid=false, max=<0 0>, min=<0 0>}
     }
     selectedTextColor {colorUseRule=UseColor_Specified, specifiedColor=<rgba 1 1 1 1>}
     selectorFocusedBrush {
      bIsDynamicallyLoaded=false
      drawAs=Border
      imageType=FullColor
      mirroring=NoMirror
      resourceName=../../../Engine/Content/Slate/Common/Selector.png
      resourceObject=null
      tiling=Both
      imageSize=<0 0>
      margin=<ltrb 0.25 0.25 0.25 0.25>
      outlineSettings {
       bUseBrushTransparency=false
       roundingType=HalfHeightRadius
       width=0
       color {colorUseRule=UseColor_Specified, specifiedColor=<rgba 0 0 0 0>}
       cornerRadii=<0 0 0 0>
      }
      tintColor {colorUseRule=UseColor_Specified, specifiedColor=<rgba 0.701 0.225 0.003 1>}
      uVRegion {bIsValid=false, max=<0 0>, min=<0 0>}
     }
     textColor {colorUseRule=UseColor_Specified, specifiedColor=<rgba 0.5271 0.5271 0.5271 1>}
    }
    widgetStyle {
     comboButtonStyle {
      downArrowAlign=VAlign_Center
      buttonStyle {
       clickedSlateSound {resourceObject=null}
       disabled {
        bIsDynamicallyLoaded=false
        drawAs=RoundedBox
        imageType=NoImage
        mirroring=NoMirror
        resourceName=None
        resourceObject=null
        tiling=NoTile
        imageSize=<0 0>
        margin=<ltrb 0 0 0 0>
        outlineSettings {
         bUseBrushTransparency=false
         roundingType=FixedRadius
         width=1
         color {colorUseRule=UseColor_ColorTable, specifiedColor=<rgba 0 0 0 0>}
         cornerRadii=<4 4 4 4>
        }
        tintColor {colorUseRule=UseColor_Specified, specifiedColor=<rgba 0.0395 0.0395 0.0395 1>}
        uVRegion {bIsValid=false, max=<0 0>, min=<0 0>}
       }
       disabledForeground {colorUseRule=UseColor_ColorTable, specifiedColor=<rgba 0 0 0 0>}
       hovered {
        bIsDynamicallyLoaded=false
        drawAs=RoundedBox
        imageType=NoImage
        mirroring=NoMirror
        resourceName=None
        resourceObject=null
        tiling=NoTile
        imageSize=<0 0>
        margin=<ltrb 0 0 0 0>
        outlineSettings {
         bUseBrushTransparency=false
         roundingType=FixedRadius
         width=1
         color {colorUseRule=UseColor_ColorTable, specifiedColor=<rgba 0 0 0 0>}
         cornerRadii=<4 4 4 4>
        }
        tintColor {colorUseRule=UseColor_Specified, specifiedColor=<rgba 0.0048 0.0048 0.0048 1>}
        uVRegion {bIsValid=false, max=<0 0>, min=<0 0>}
       }
       hoveredForeground {colorUseRule=UseColor_ColorTable, specifiedColor=<rgba 0 0 0 0>}
       hoveredSlateSound {resourceObject=null}
       normal {
        bIsDynamicallyLoaded=false
        drawAs=RoundedBox
        imageType=NoImage
        mirroring=NoMirror
        resourceName=None
        resourceObject=null
        tiling=NoTile
        imageSize=<0 0>
        margin=<ltrb 0 0 0 0>
        outlineSettings {
         bUseBrushTransparency=false
         roundingType=FixedRadius
         width=1
         color {colorUseRule=UseColor_ColorTable, specifiedColor=<rgba 0 0 0 0>}
         cornerRadii=<4 4 4 4>
        }
        tintColor {colorUseRule=UseColor_Specified, specifiedColor=<rgba 0.0048 0.0048 0.0048 1>}
        uVRegion {bIsValid=false, max=<0 0>, min=<0 0>}
       }
       normalForeground {colorUseRule=UseColor_ColorTable, specifiedColor=<rgba 0 0 0 0>}
       normalPadding=<ltrb 8 1 8 1>
       pressed {
        bIsDynamicallyLoaded=false
        drawAs=RoundedBox
        imageType=NoImage
        mirroring=NoMirror
        resourceName=None
        resourceObject=null
        tiling=NoTile
        imageSize=<0 0>
        margin=<ltrb 0 0 0 0>
        outlineSettings {
         bUseBrushTransparency=false
         roundingType=FixedRadius
         width=1
         color {colorUseRule=UseColor_ColorTable, specifiedColor=<rgba 0 0 0 0>}
         cornerRadii=<4 4 4 4>
        }
        tintColor {colorUseRule=UseColor_Specified, specifiedColor=<rgba 0.0048 0.0048 0.0048 1>}
        uVRegion {bIsValid=false, max=<0 0>, min=<0 0>}
       }
       pressedForeground {colorUseRule=UseColor_ColorTable, specifiedColor=<rgba 0 0 0 0>}
       pressedPadding=<ltrb 8 2 8 0>
       pressedSlateSound {resourceObject=null}
      }
      contentPadding=<ltrb 0 0 0 0>
      downArrowImage {
       bIsDynamicallyLoaded=false
       drawAs=Image
       imageType=Vector
       mirroring=NoMirror
       resourceName=../../../Engine/Content/Slate/Starship/CoreWidgets/ComboBox/wide-chevron-down.svg
       resourceObject=null
       tiling=NoTile
       imageSize=<16 16>
       margin=<ltrb 0 0 0 0>
       outlineSettings {
        bUseBrushTransparency=false
        roundingType=HalfHeightRadius
        width=0
        color {colorUseRule=UseColor_Specified, specifiedColor=<rgba 0 0 0 0>}
        cornerRadii=<0 0 0 0>
       }
       tintColor {colorUseRule=UseColor_Specified, specifiedColor=<rgba 1 1 1 1>}
       uVRegion {bIsValid=false, max=<0 0>, min=<0 0>}
      }
      downArrowPadding=<ltrb 2 2 2 2>
      menuBorderBrush {
       bIsDynamicallyLoaded=false
       drawAs=RoundedBox
       imageType=NoImage
       mirroring=NoMirror
       resourceName=None
       resourceObject=null
       tiling=NoTile
       imageSize=<0 0>
       margin=<ltrb 0 0 0 0>
       outlineSettings {
        bUseBrushTransparency=false
        roundingType=FixedRadius
        width=1
        color {colorUseRule=UseColor_ColorTable, specifiedColor=<rgba 0 0 0 0>}
        cornerRadii=<0 0 0 0>
       }
       tintColor {colorUseRule=UseColor_Specified, specifiedColor=<rgba 0.0395 0.0395 0.0395 1>}
       uVRegion {bIsValid=false, max=<0 0>, min=<0 0>}
      }
      menuBorderPadding=<ltrb 0 0 0 0>
      shadowColorAndOpacity=<rgba 0 0 0 1>
      shadowOffset=<0 0>
     }
     contentPadding=<ltrb 0 0 0 0>
     menuRowPadding=<ltrb 8 3 8 3>
     pressedSlateSound {resourceObject=null}
     selectionChangeSlateSound {resourceObject=null}
    }
   }
   slot {class=SizeBoxSlot}
  }
  TypeImage {
   widgetClassPath=&Image
   bIsVariable=true
   parent=HorizontalBox_1
   widget=TypeImage
   properties {brush {
     bIsDynamicallyLoaded=false
     drawAs=Image
     imageType=FullColor
     mirroring=NoMirror
     resourceName=None
     resourceObject=/AnimInEdTools/AnimationBrowser/Icons/Icon_Categories
     tiling=NoTile
     imageSize=<18 28>
     margin=<ltrb 0 0 0 0>
     outlineSettings {
      bUseBrushTransparency=false
      roundingType=HalfHeightRadius
      width=0
      color {colorUseRule=UseColor_Specified, specifiedColor=<rgba 0 0 0 0>}
      cornerRadii=<0 0 0 0>
     }
     tintColor {colorUseRule=UseColor_Specified, specifiedColor=<rgba 1 1 1 1>}
     uVRegion {bIsValid=false, max=<0 0>, min=<0 0>}
    }}
   slot {class=HorizontalBoxSlot, properties {horizontalAlignment=HAlign_Center, verticalAlignment=VAlign_Center}}
  }
  SizeBox_2 {
   widgetClassPath=&SizeBox
   parent=HorizontalBox_1
   widget=SizeBox_2
   properties {
    bOverride_MaxDesiredHeight=true
    bOverride_MaxDesiredWidth=true
    bOverride_MinDesiredWidth=true
    maxDesiredHeight=25
    maxDesiredWidth=100
    minDesiredWidth=50
   }
   slot {class=HorizontalBoxSlot, properties {size {sizeRule=Fill, value=1}}}
  }
  Subfolder_Combobox {
   widgetClassPath=&SearchableStringComboBox
   bIsVariable=true
   parent=SizeBox_2
   widget=Subfolder_Combobox
   properties {
    selectedOption=Subfolder
    defaultOptions [Subfolder]
    font {
     bForceMonospaced=false
     bMaterialIsStencil=false
     fontMaterial=null
     fontObject=&Roboto
     letterSpacing=0
     monospacedWidth=1
     size=12
     skewAmount=0
     typefaceFontName=Bold
     outlineSettings {
      bApplyOutlineToDropShadows=false
      bMiteredCorners=false
      bSeparateFillAlpha=false
      outlineMaterial=null
      outlineSize=0
      outlineColor=<rgba 0 0 0 1>
     }
    }
    foregroundColor {colorUseRule=UseColor_Specified, specifiedColor=<rgba 0.5271 0.5271 0.5271 1>}
    itemStyle {
     bUseParentRowBrush=false
     activeBrush {
      bIsDynamicallyLoaded=false
      drawAs=Image
      imageType=NoImage
      mirroring=NoMirror
      resourceName=None
      resourceObject=null
      tiling=NoTile
      imageSize=<0 0>
      margin=<ltrb 0 0 0 0>
      outlineSettings {
       bUseBrushTransparency=false
       roundingType=HalfHeightRadius
       width=0
       color {colorUseRule=UseColor_Specified, specifiedColor=<rgba 0 0 0 0>}
       cornerRadii=<0 0 0 0>
      }
      tintColor {colorUseRule=UseColor_Specified, specifiedColor=<rgba 0 0.162 0.7454 1>}
      uVRegion {bIsValid=false, max=<0 0>, min=<0 0>}
     }
     activeHighlightedBrush {
      bIsDynamicallyLoaded=false
      drawAs=Image
      imageType=NoImage
      mirroring=NoMirror
      resourceName=None
      resourceObject=null
      tiling=NoTile
      imageSize=<0 0>
      margin=<ltrb 0 0 0 0>
      outlineSettings {
       bUseBrushTransparency=false
       roundingType=HalfHeightRadius
       width=0
       color {colorUseRule=UseColor_Specified, specifiedColor=<rgba 0 0 0 0>}
       cornerRadii=<0 0 0 0>
      }
      tintColor {colorUseRule=UseColor_Specified, specifiedColor=<rgba 0.0252 0.0319 0.0423 1>}
      uVRegion {bIsValid=false, max=<0 0>, min=<0 0>}
     }
     activeHoveredBrush {
      bIsDynamicallyLoaded=false
      drawAs=Image
      imageType=NoImage
      mirroring=NoMirror
      resourceName=None
      resourceObject=null
      tiling=NoTile
      imageSize=<0 0>
      margin=<ltrb 0 0 0 0>
      outlineSettings {
       bUseBrushTransparency=false
       roundingType=HalfHeightRadius
       width=0
       color {colorUseRule=UseColor_Specified, specifiedColor=<rgba 0 0 0 0>}
       cornerRadii=<0 0 0 0>
      }
      tintColor {colorUseRule=UseColor_Specified, specifiedColor=<rgba 0 0.162 0.7454 1>}
      uVRegion {bIsValid=false, max=<0 0>, min=<0 0>}
     }
     dropIndicator_Above {
      bIsDynamicallyLoaded=false
      drawAs=Box
      imageType=FullColor
      mirroring=NoMirror
      resourceName=../../../Engine/Content/Slate/Common/DropZoneIndicator_Above.png
      resourceObject=null
      tiling=NoTile
      imageSize=<0 0>
      margin=<ltrb 0.625 0.625 0 0>
      outlineSettings {
       bUseBrushTransparency=false
       roundingType=HalfHeightRadius
       width=0
       color {colorUseRule=UseColor_Specified, specifiedColor=<rgba 0 0 0 0>}
       cornerRadii=<0 0 0 0>
      }
      tintColor {colorUseRule=UseColor_Specified, specifiedColor=<rgba 0.0091 0.3515 0.9647 1>}
      uVRegion {bIsValid=false, max=<0 0>, min=<0 0>}
     }
     dropIndicator_Below {
      bIsDynamicallyLoaded=false
      drawAs=Box
      imageType=FullColor
      mirroring=NoMirror
      resourceName=../../../Engine/Content/Slate/Common/DropZoneIndicator_Below.png
      resourceObject=null
      tiling=NoTile
      imageSize=<0 0>
      margin=<ltrb 0.625 0 0 0.625>
      outlineSettings {
       bUseBrushTransparency=false
       roundingType=HalfHeightRadius
       width=0
       color {colorUseRule=UseColor_Specified, specifiedColor=<rgba 0 0 0 0>}
       cornerRadii=<0 0 0 0>
      }
      tintColor {colorUseRule=UseColor_Specified, specifiedColor=<rgba 0.0091 0.3515 0.9647 1>}
      uVRegion {bIsValid=false, max=<0 0>, min=<0 0>}
     }
     dropIndicator_Onto {
      bIsDynamicallyLoaded=false
      drawAs=Box
      imageType=FullColor
      mirroring=NoMirror
      resourceName=../../../Engine/Content/Slate/Common/DropZoneIndicator_Onto.png
      resourceObject=null
      tiling=NoTile
      imageSize=<0 0>
      margin=<ltrb 0.25 0.25 0.25 0.25>
      outlineSettings {
       bUseBrushTransparency=false
       roundingType=HalfHeightRadius
       width=0
       color {colorUseRule=UseColor_Specified, specifiedColor=<rgba 0 0 0 0>}
       cornerRadii=<0 0 0 0>
      }
      tintColor {colorUseRule=UseColor_Specified, specifiedColor=<rgba 0.0091 0.3515 0.9647 1>}
      uVRegion {bIsValid=false, max=<0 0>, min=<0 0>}
     }
     evenRowBackgroundBrush {
      bIsDynamicallyLoaded=false
      drawAs=Image
      imageType=NoImage
      mirroring=NoMirror
      resourceName=None
      resourceObject=null
      tiling=NoTile
      imageSize=<0 0>
      margin=<ltrb 0 0 0 0>
      outlineSettings {
       bUseBrushTransparency=false
       roundingType=HalfHeightRadius
       width=0
       color {colorUseRule=UseColor_Specified, specifiedColor=<rgba 0 0 0 0>}
       cornerRadii=<0 0 0 0>
      }
      tintColor {colorUseRule=UseColor_Specified, specifiedColor=<rgba 0.0103 0.0103 0.0103 1>}
      uVRegion {bIsValid=false, max=<0 0>, min=<0 0>}
     }
     evenRowBackgroundHoveredBrush {
      bIsDynamicallyLoaded=false
      drawAs=Image
      imageType=NoImage
      mirroring=NoMirror
      resourceName=None
      resourceObject=null
      tiling=NoTile
      imageSize=<0 0>
      margin=<ltrb 0 0 0 0>
      outlineSettings {
       bUseBrushTransparency=false
       roundingType=HalfHeightRadius
       width=0
       color {colorUseRule=UseColor_Specified, specifiedColor=<rgba 0 0 0 0>}
       cornerRadii=<0 0 0 0>
      }
      tintColor {colorUseRule=UseColor_Specified, specifiedColor=<rgba 0.0176 0.0176 0.0176 1>}
      uVRegion {bIsValid=false, max=<0 0>, min=<0 0>}
     }
     inactiveBrush {
      bIsDynamicallyLoaded=false
      drawAs=Image
      imageType=NoImage
      mirroring=NoMirror
      resourceName=None
      resourceObject=null
      tiling=NoTile
      imageSize=<0 0>
      margin=<ltrb 0 0 0 0>
      outlineSettings {
       bUseBrushTransparency=false
       roundingType=HalfHeightRadius
       width=0
       color {colorUseRule=UseColor_Specified, specifiedColor=<rgba 0 0 0 0>}
       cornerRadii=<0 0 0 0>
      }
      tintColor {colorUseRule=UseColor_Specified, specifiedColor=<rgba 0.0513 0.0953 0.159 1>}
      uVRegion {bIsValid=false, max=<0 0>, min=<0 0>}
     }
     inactiveHighlightedBrush {
      bIsDynamicallyLoaded=false
      drawAs=Image
      imageType=NoImage
      mirroring=NoMirror
      resourceName=None
      resourceObject=null
      tiling=NoTile
      imageSize=<0 0>
      margin=<ltrb 0 0 0 0>
      outlineSettings {
       bUseBrushTransparency=false
       roundingType=HalfHeightRadius
       width=0
       color {colorUseRule=UseColor_Specified, specifiedColor=<rgba 0 0 0 0>}
       cornerRadii=<0 0 0 0>
      }
      tintColor {colorUseRule=UseColor_Specified, specifiedColor=<rgba 0.0252 0.0319 0.0423 1>}
      uVRegion {bIsValid=false, max=<0 0>, min=<0 0>}
     }
     inactiveHoveredBrush {
      bIsDynamicallyLoaded=false
      drawAs=Image
      imageType=NoImage
      mirroring=NoMirror
      resourceName=None
      resourceObject=null
      tiling=NoTile
      imageSize=<0 0>
      margin=<ltrb 0 0 0 0>
      outlineSettings {
       bUseBrushTransparency=false
       roundingType=HalfHeightRadius
       width=0
       color {colorUseRule=UseColor_Specified, specifiedColor=<rgba 0 0 0 0>}
       cornerRadii=<0 0 0 0>
      }
      tintColor {colorUseRule=UseColor_Specified, specifiedColor=<rgba 0.0513 0.0953 0.159 1>}
      uVRegion {bIsValid=false, max=<0 0>, min=<0 0>}
     }
     oddRowBackgroundBrush {
      bIsDynamicallyLoaded=false
      drawAs=Image
      imageType=NoImage
      mirroring=NoMirror
      resourceName=None
      resourceObject=null
      tiling=NoTile
      imageSize=<0 0>
      margin=<ltrb 0 0 0 0>
      outlineSettings {
       bUseBrushTransparency=false
       roundingType=HalfHeightRadius
       width=0
       color {colorUseRule=UseColor_Specified, specifiedColor=<rgba 0 0 0 0>}
       cornerRadii=<0 0 0 0>
      }
      tintColor {colorUseRule=UseColor_Specified, specifiedColor=<rgba 0.0103 0.0103 0.0103 1>}
      uVRegion {bIsValid=false, max=<0 0>, min=<0 0>}
     }
     oddRowBackgroundHoveredBrush {
      bIsDynamicallyLoaded=false
      drawAs=Image
      imageType=NoImage
      mirroring=NoMirror
      resourceName=None
      resourceObject=null
      tiling=NoTile
      imageSize=<0 0>
      margin=<ltrb 0 0 0 0>
      outlineSettings {
       bUseBrushTransparency=false
       roundingType=HalfHeightRadius
       width=0
       color {colorUseRule=UseColor_Specified, specifiedColor=<rgba 0 0 0 0>}
       cornerRadii=<0 0 0 0>
      }
      tintColor {colorUseRule=UseColor_Specified, specifiedColor=<rgba 0.0176 0.0176 0.0176 1>}
      uVRegion {bIsValid=false, max=<0 0>, min=<0 0>}
     }
     parentRowBackgroundBrush {
      bIsDynamicallyLoaded=false
      drawAs=Image
      imageType=NoImage
      mirroring=NoMirror
      resourceName=None
      resourceObject=null
      tiling=NoTile
      imageSize=<32 32>
      margin=<ltrb 0 0 0 0>
      outlineSettings {
       bUseBrushTransparency=false
       roundingType=HalfHeightRadius
       width=0
       color {colorUseRule=UseColor_Specified, specifiedColor=<rgba 0 0 0 0>}
       cornerRadii=<0 0 0 0>
      }
      tintColor {colorUseRule=UseColor_Specified, specifiedColor=<rgba 1 1 1 1>}
      uVRegion {bIsValid=false, max=<0 0>, min=<0 0>}
     }
     parentRowBackgroundHoveredBrush {
      bIsDynamicallyLoaded=false
      drawAs=Image
      imageType=NoImage
      mirroring=NoMirror
      resourceName=None
      resourceObject=null
      tiling=NoTile
      imageSize=<32 32>
      margin=<ltrb 0 0 0 0>
      outlineSettings {
       bUseBrushTransparency=false
       roundingType=HalfHeightRadius
       width=0
       color {colorUseRule=UseColor_Specified, specifiedColor=<rgba 0 0 0 0>}
       cornerRadii=<0 0 0 0>
      }
      tintColor {colorUseRule=UseColor_Specified, specifiedColor=<rgba 1 1 1 1>}
      uVRegion {bIsValid=false, max=<0 0>, min=<0 0>}
     }
     selectedTextColor {colorUseRule=UseColor_Specified, specifiedColor=<rgba 1 1 1 1>}
     selectorFocusedBrush {
      bIsDynamicallyLoaded=false
      drawAs=Border
      imageType=FullColor
      mirroring=NoMirror
      resourceName=../../../Engine/Content/Slate/Common/Selector.png
      resourceObject=null
      tiling=Both
      imageSize=<0 0>
      margin=<ltrb 0.25 0.25 0.25 0.25>
      outlineSettings {
       bUseBrushTransparency=false
       roundingType=HalfHeightRadius
       width=0
       color {colorUseRule=UseColor_Specified, specifiedColor=<rgba 0 0 0 0>}
       cornerRadii=<0 0 0 0>
      }
      tintColor {colorUseRule=UseColor_Specified, specifiedColor=<rgba 0.701 0.225 0.003 1>}
      uVRegion {bIsValid=false, max=<0 0>, min=<0 0>}
     }
     textColor {colorUseRule=UseColor_Specified, specifiedColor=<rgba 0.5271 0.5271 0.5271 1>}
    }
    widgetStyle {
     comboButtonStyle {
      downArrowAlign=VAlign_Center
      buttonStyle {
       clickedSlateSound {resourceObject=null}
       disabled {
        bIsDynamicallyLoaded=false
        drawAs=RoundedBox
        imageType=NoImage
        mirroring=NoMirror
        resourceName=None
        resourceObject=null
        tiling=NoTile
        imageSize=<0 0>
        margin=<ltrb 0 0 0 0>
        outlineSettings {
         bUseBrushTransparency=false
         roundingType=FixedRadius
         width=1
         color {colorUseRule=UseColor_ColorTable, specifiedColor=<rgba 0 0 0 0>}
         cornerRadii=<4 4 4 4>
        }
        tintColor {colorUseRule=UseColor_Specified, specifiedColor=<rgba 0.0395 0.0395 0.0395 1>}
        uVRegion {bIsValid=false, max=<0 0>, min=<0 0>}
       }
       disabledForeground {colorUseRule=UseColor_ColorTable, specifiedColor=<rgba 0 0 0 0>}
       hovered {
        bIsDynamicallyLoaded=false
        drawAs=RoundedBox
        imageType=NoImage
        mirroring=NoMirror
        resourceName=None
        resourceObject=null
        tiling=NoTile
        imageSize=<0 0>
        margin=<ltrb 0 0 0 0>
        outlineSettings {
         bUseBrushTransparency=false
         roundingType=FixedRadius
         width=1
         color {colorUseRule=UseColor_ColorTable, specifiedColor=<rgba 0 0 0 0>}
         cornerRadii=<4 4 4 4>
        }
        tintColor {colorUseRule=UseColor_Specified, specifiedColor=<rgba 0.0048 0.0048 0.0048 1>}
        uVRegion {bIsValid=false, max=<0 0>, min=<0 0>}
       }
       hoveredForeground {colorUseRule=UseColor_ColorTable, specifiedColor=<rgba 0 0 0 0>}
       hoveredSlateSound {resourceObject=null}
       normal {
        bIsDynamicallyLoaded=false
        drawAs=RoundedBox
        imageType=NoImage
        mirroring=NoMirror
        resourceName=None
        resourceObject=null
        tiling=NoTile
        imageSize=<0 0>
        margin=<ltrb 0 0 0 0>
        outlineSettings {
         bUseBrushTransparency=false
         roundingType=FixedRadius
         width=1
         color {colorUseRule=UseColor_ColorTable, specifiedColor=<rgba 0 0 0 0>}
         cornerRadii=<4 4 4 4>
        }
        tintColor {colorUseRule=UseColor_Specified, specifiedColor=<rgba 0.0048 0.0048 0.0048 1>}
        uVRegion {bIsValid=false, max=<0 0>, min=<0 0>}
       }
       normalForeground {colorUseRule=UseColor_ColorTable, specifiedColor=<rgba 0 0 0 0>}
       normalPadding=<ltrb 8 1 8 1>
       pressed {
        bIsDynamicallyLoaded=false
        drawAs=RoundedBox
        imageType=NoImage
        mirroring=NoMirror
        resourceName=None
        resourceObject=null
        tiling=NoTile
        imageSize=<0 0>
        margin=<ltrb 0 0 0 0>
        outlineSettings {
         bUseBrushTransparency=false
         roundingType=FixedRadius
         width=1
         color {colorUseRule=UseColor_ColorTable, specifiedColor=<rgba 0 0 0 0>}
         cornerRadii=<4 4 4 4>
        }
        tintColor {colorUseRule=UseColor_Specified, specifiedColor=<rgba 0.0048 0.0048 0.0048 1>}
        uVRegion {bIsValid=false, max=<0 0>, min=<0 0>}
       }
       pressedForeground {colorUseRule=UseColor_ColorTable, specifiedColor=<rgba 0 0 0 0>}
       pressedPadding=<ltrb 8 2 8 0>
       pressedSlateSound {resourceObject=null}
      }
      contentPadding=<ltrb 0 0 0 0>
      downArrowImage {
       bIsDynamicallyLoaded=false
       drawAs=Image
       imageType=Vector
       mirroring=NoMirror
       resourceName=../../../Engine/Content/Slate/Starship/CoreWidgets/ComboBox/wide-chevron-down.svg
       resourceObject=null
       tiling=NoTile
       imageSize=<16 16>
       margin=<ltrb 0 0 0 0>
       outlineSettings {
        bUseBrushTransparency=false
        roundingType=HalfHeightRadius
        width=0
        color {colorUseRule=UseColor_Specified, specifiedColor=<rgba 0 0 0 0>}
        cornerRadii=<0 0 0 0>
       }
       tintColor {colorUseRule=UseColor_Specified, specifiedColor=<rgba 1 1 1 1>}
       uVRegion {bIsValid=false, max=<0 0>, min=<0 0>}
      }
      downArrowPadding=<ltrb 2 2 2 2>
      menuBorderBrush {
       bIsDynamicallyLoaded=false
       drawAs=RoundedBox
       imageType=NoImage
       mirroring=NoMirror
       resourceName=None
       resourceObject=null
       tiling=NoTile
       imageSize=<0 0>
       margin=<ltrb 0 0 0 0>
       outlineSettings {
        bUseBrushTransparency=false
        roundingType=FixedRadius
        width=1
        color {colorUseRule=UseColor_ColorTable, specifiedColor=<rgba 0 0 0 0>}
        cornerRadii=<0 0 0 0>
       }
       tintColor {colorUseRule=UseColor_Specified, specifiedColor=<rgba 0.0395 0.0395 0.0395 1>}
       uVRegion {bIsValid=false, max=<0 0>, min=<0 0>}
      }
      menuBorderPadding=<ltrb 0 0 0 0>
      shadowColorAndOpacity=<rgba 0 0 0 1>
      shadowOffset=<0 0>
     }
     contentPadding=<ltrb 0 0 0 0>
     menuRowPadding=<ltrb 8 3 8 3>
     pressedSlateSound {resourceObject=null}
     selectionChangeSlateSound {resourceObject=null}
    }
   }
   slot {class=SizeBoxSlot}
  }
  FilterButton {
   widgetClassPath=&Button
   parent=HorizontalBox_1
   widget=FilterButton
   properties {
    toolTipText=Settings Menu
    widgetStyle {
     clickedSlateSound {resourceObject=null}
     disabled {
      bIsDynamicallyLoaded=false
      drawAs=RoundedBox
      imageType=NoImage
      mirroring=NoMirror
      resourceName=None
      resourceObject=null
      tiling=NoTile
      imageSize=<0 0>
      margin=<ltrb 0 0 0 0>
      outlineSettings {
       bUseBrushTransparency=false
       roundingType=FixedRadius
       width=1
       color {colorUseRule=UseColor_ColorTable, specifiedColor=<rgba 0 0 0 0>}
       cornerRadii=<4 4 4 4>
      }
      tintColor {colorUseRule=UseColor_Specified, specifiedColor=<rgba 0.0395 0.0395 0.0395 1>}
      uVRegion {bIsValid=false, max=<0 0>, min=<0 0>}
     }
     disabledForeground {colorUseRule=UseColor_ColorTable, specifiedColor=<rgba 0 0 0 0>}
     hovered {
      bIsDynamicallyLoaded=false
      drawAs=RoundedBox
      imageType=FullColor
      mirroring=NoMirror
      resourceName=None
      resourceObject=null
      tiling=NoTile
      imageSize=<28 28>
      margin=<ltrb 0 0 0 0>
      outlineSettings {
       bUseBrushTransparency=false
       roundingType=FixedRadius
       width=0
       color {colorUseRule=UseColor_ColorTable, specifiedColor=<rgba 0 0 0 0>}
       cornerRadii=<4 4 4 4>
      }
      tintColor {colorUseRule=UseColor_Specified, specifiedColor=<rgba 0.0941 0.0941 0.0941 1>}
      uVRegion {bIsValid=false, max=<0 0>, min=<0 0>}
     }
     hoveredForeground {colorUseRule=UseColor_ColorTable, specifiedColor=<rgba 0 0 0 0>}
     hoveredSlateSound {resourceObject=null}
     normal {
      bIsDynamicallyLoaded=false
      drawAs=NoDrawType
      imageType=FullColor
      mirroring=NoMirror
      resourceName=None
      resourceObject=null
      tiling=NoTile
      imageSize=<28 28>
      margin=<ltrb 0 0 0 0>
      outlineSettings {
       bUseBrushTransparency=false
       roundingType=FixedRadius
       width=0
       color {colorUseRule=UseColor_ColorTable, specifiedColor=<rgba 0 0 0 0>}
       cornerRadii=<4 4 4 4>
      }
      tintColor {colorUseRule=UseColor_Specified, specifiedColor=<rgba 0.0176 0.0176 0.0176 1>}
      uVRegion {bIsValid=false, max=<0 0>, min=<0 0>}
     }
     normalForeground {colorUseRule=UseColor_ColorTable, specifiedColor=<rgba 0 0 0 0>}
     normalPadding=<ltrb 0 0 0 0>
     pressed {
      bIsDynamicallyLoaded=false
      drawAs=RoundedBox
      imageType=FullColor
      mirroring=NoMirror
      resourceName=None
      resourceObject=null
      tiling=NoTile
      imageSize=<28 28>
      margin=<ltrb 0 0 0 0>
      outlineSettings {
       bUseBrushTransparency=false
       roundingType=FixedRadius
       width=0
       color {colorUseRule=UseColor_ColorTable, specifiedColor=<rgba 0 0 0 0>}
       cornerRadii=<4 4 4 4>
      }
      tintColor {colorUseRule=UseColor_Specified, specifiedColor=<rgba 0.0392 0.0392 0.0392 1>}
      uVRegion {bIsValid=false, max=<0 0>, min=<0 0>}
     }
     pressedForeground {colorUseRule=UseColor_ColorTable, specifiedColor=<rgba 0 0 0 0>}
     pressedPadding=<ltrb 0 0 0 0>
     pressedSlateSound {resourceObject=null}
    }
   }
   slot {class=HorizontalBoxSlot, properties {horizontalAlignment=HAlign_Center, padding=<ltrb 2 0 0 0>}}
  }
  HorizontalBox_4 {
   widgetClassPath=&HorizontalBox
   parent=FilterButton
   widget=HorizontalBox_4
   slot {class=ButtonSlot, properties {padding=<ltrb 3 0 3 0>}}
  }
  AssetImage_3 {
   widgetClassPath=&Image
   bIsVariable=true
   parent=HorizontalBox_4
   widget=AssetImage_3
   properties {brush {
     bIsDynamicallyLoaded=false
     drawAs=Image
     imageType=FullColor
     mirroring=NoMirror
     resourceName=None
     resourceObject=/AnimInEdTools/AnimationBrowser/Icons/Icon_AnimationBrowser_Filter
     tiling=NoTile
     imageSize=<14 14>
     margin=<ltrb 0 0 0 0>
     outlineSettings {
      bUseBrushTransparency=false
      roundingType=HalfHeightRadius
      width=0
      color {colorUseRule=UseColor_Specified, specifiedColor=<rgba 0 0 0 0>}
      cornerRadii=<0 0 0 0>
     }
     tintColor {colorUseRule=UseColor_Specified, specifiedColor=<rgba 1 1 1 1>}
     uVRegion {bIsValid=false, max=<0 0>, min=<0 0>}
    }}
   slot {class=HorizontalBoxSlot, properties {horizontalAlignment=HAlign_Center, verticalAlignment=VAlign_Center}}
  }
  HorizontalBox_0 {
   widgetClassPath=&HorizontalBox
   parent=VerticalBox_1
   widget=HorizontalBox_0
   slot {class=VerticalBoxSlot, properties {padding=<ltrb 0 3 0 0>, size {sizeRule=Fill, value=1}}}
  }
  AddAnimationButton {
   widgetClassPath=&Button
   parent=HorizontalBox_0
   widget=AddAnimationButton
   properties {
    bIsEnabled=false
    toolTipText=Create a new animation
    widgetStyle {
     clickedSlateSound {resourceObject=null}
     disabled {
      bIsDynamicallyLoaded=false
      drawAs=RoundedBox
      imageType=NoImage
      mirroring=NoMirror
      resourceName=None
      resourceObject=null
      tiling=NoTile
      imageSize=<0 0>
      margin=<ltrb 0 0 0 0>
      outlineSettings {
       bUseBrushTransparency=false
       roundingType=FixedRadius
       width=1
       color {colorUseRule=UseColor_ColorTable, specifiedColor=<rgba 0 0 0 0>}
       cornerRadii=<4 4 4 4>
      }
      tintColor {colorUseRule=UseColor_Specified, specifiedColor=<rgba 0.0395 0.0395 0.0395 1>}
      uVRegion {bIsValid=false, max=<0 0>, min=<0 0>}
     }
     disabledForeground {colorUseRule=UseColor_ColorTable, specifiedColor=<rgba 0 0 0 0>}
     hovered {
      bIsDynamicallyLoaded=false
      drawAs=RoundedBox
      imageType=NoImage
      mirroring=NoMirror
      resourceName=None
      resourceObject=null
      tiling=NoTile
      imageSize=<32 32>
      margin=<ltrb 0 0 0 0>
      outlineSettings {
       bUseBrushTransparency=false
       roundingType=FixedRadius
       width=0
       color {colorUseRule=UseColor_ColorTable, specifiedColor=<rgba 0 0 0 0>}
       cornerRadii=<4 4 4 4>
      }
      tintColor {colorUseRule=UseColor_Specified, specifiedColor=<rgba 0.0953 0.0953 0.0953 1>}
      uVRegion {bIsValid=false, max=<0 0>, min=<0 0>}
     }
     hoveredForeground {colorUseRule=UseColor_ColorTable, specifiedColor=<rgba 0 0 0 0>}
     hoveredSlateSound {resourceObject=null}
     normal {
      bIsDynamicallyLoaded=false
      drawAs=NoDrawType
      imageType=FullColor
      mirroring=NoMirror
      resourceName=None
      resourceObject=null
      tiling=NoTile
      imageSize=<14 14>
      margin=<ltrb 0 0 0 0>
      outlineSettings {
       bUseBrushTransparency=false
       roundingType=FixedRadius
       width=1
       color {colorUseRule=UseColor_ColorTable, specifiedColor=<rgba 0 0 0 0>}
       cornerRadii=<4 4 4 4>
      }
      tintColor {colorUseRule=UseColor_Specified, specifiedColor=<rgba 0.0395 0.0395 0.0395 1>}
      uVRegion {bIsValid=false, max=<0 0>, min=<0 0>}
     }
     normalForeground {colorUseRule=UseColor_ColorTable, specifiedColor=<rgba 0 0 0 0>}
     normalPadding=<ltrb 12 1.5 12 1.5>
     pressed {
      bIsDynamicallyLoaded=false
      drawAs=RoundedBox
      imageType=NoImage
      mirroring=NoMirror
      resourceName=None
      resourceObject=null
      tiling=NoTile
      imageSize=<32 32>
      margin=<ltrb 0 0 0 0>
      outlineSettings {
       bUseBrushTransparency=false
       roundingType=FixedRadius
       width=0
       color {colorUseRule=UseColor_ColorTable, specifiedColor=<rgba 0 0 0 0>}
       cornerRadii=<4 4 4 4>
      }
      tintColor {colorUseRule=UseColor_Specified, specifiedColor=<rgba 0.0284 0.0284 0.0284 1>}
      uVRegion {bIsValid=false, max=<0 0>, min=<0 0>}
     }
     pressedForeground {colorUseRule=UseColor_ColorTable, specifiedColor=<rgba 0 0 0 0>}
     pressedPadding=<ltrb 12 2.5 12 0.5>
     pressedSlateSound {resourceObject=null}
    }
   }
   slot {class=HorizontalBoxSlot, properties {horizontalAlignment=HAlign_Right, padding=<ltrb 0 0 5 0>}}
  }
  HorizontalBox_3 {
   widgetClassPath=&HorizontalBox
   parent=AddAnimationButton
   widget=HorizontalBox_3
   slot {class=ButtonSlot, properties {padding=<ltrb 0 0 0 0>}}
  }
  AssetImage_1 {
   widgetClassPath=&Image
   bIsVariable=true
   parent=HorizontalBox_3
   widget=AssetImage_1
   properties {brush {
     bIsDynamicallyLoaded=false
     drawAs=Image
     imageType=FullColor
     mirroring=NoMirror
     resourceName=None
     resourceObject=/AnimInEdTools/AnimationBrowser/Icons/Icon_LevelSequence_add
     tiling=NoTile
     imageSize=<14 14>
     margin=<ltrb 0 0 0 0>
     outlineSettings {
      bUseBrushTransparency=false
      roundingType=HalfHeightRadius
      width=0
      color {colorUseRule=UseColor_Specified, specifiedColor=<rgba 0 0 0 0>}
      cornerRadii=<0 0 0 0>
     }
     tintColor {colorUseRule=UseColor_Specified, specifiedColor=<rgba 0.2108 1 0 1>}
     uVRegion {bIsValid=false, max=<0 0>, min=<0 0>}
    }}
   slot {class=HorizontalBoxSlot, properties {horizontalAlignment=HAlign_Center, verticalAlignment=VAlign_Center, padding=<ltrb 0 0 5 0>}}
  }
  TextBlock_1 {
   widgetClassPath=&TextBlock
   parent=HorizontalBox_3
   widget=TextBlock_1
   properties {
    autoWrapText=true
    justification=Center
    text=New
    colorAndOpacity {colorUseRule=UseColor_Specified, specifiedColor=<rgba 0.5255 0.5255 0.5255 1>}
    font {
     bForceMonospaced=false
     bMaterialIsStencil=false
     fontMaterial=null
     fontObject=&Roboto
     letterSpacing=0
     monospacedWidth=1
     size=10
     skewAmount=0
     typefaceFontName=Bold
     outlineSettings {
      bApplyOutlineToDropShadows=false
      bMiteredCorners=false
      bSeparateFillAlpha=false
      outlineMaterial=null
      outlineSize=0
      outlineColor=<rgba 0 0 0 1>
     }
    }
   }
   slot {class=HorizontalBoxSlot}
  }
  SearchTextBox {
   widgetClassPath=&EditorUtilityEditableTextBox
   bIsVariable=true
   parent=HorizontalBox_0
   widget=SearchTextBox
   properties {hintText=Search}
   slot {class=HorizontalBoxSlot, properties {size {sizeRule=Fill, value=1}}}
  }
  ClearSearchBox {
   widgetClassPath=&Button
   parent=HorizontalBox_0
   widget=ClearSearchBox
   properties {
    toolTipText=Clear Selection
    widgetStyle {
     clickedSlateSound {resourceObject=null}
     disabled {
      bIsDynamicallyLoaded=false
      drawAs=RoundedBox
      imageType=NoImage
      mirroring=NoMirror
      resourceName=None
      resourceObject=null
      tiling=NoTile
      imageSize=<0 0>
      margin=<ltrb 0 0 0 0>
      outlineSettings {
       bUseBrushTransparency=false
       roundingType=FixedRadius
       width=1
       color {colorUseRule=UseColor_ColorTable, specifiedColor=<rgba 0 0 0 0>}
       cornerRadii=<4 4 4 4>
      }
      tintColor {colorUseRule=UseColor_Specified, specifiedColor=<rgba 0.0395 0.0395 0.0395 1>}
      uVRegion {bIsValid=false, max=<0 0>, min=<0 0>}
     }
     disabledForeground {colorUseRule=UseColor_ColorTable, specifiedColor=<rgba 0 0 0 0>}
     hovered {
      bIsDynamicallyLoaded=false
      drawAs=RoundedBox
      imageType=NoImage
      mirroring=NoMirror
      resourceName=None
      resourceObject=null
      tiling=NoTile
      imageSize=<32 32>
      margin=<ltrb 0 0 0 0>
      outlineSettings {
       bUseBrushTransparency=false
       roundingType=FixedRadius
       width=0
       color {colorUseRule=UseColor_ColorTable, specifiedColor=<rgba 0 0 0 0>}
       cornerRadii=<4 4 4 4>
      }
      tintColor {colorUseRule=UseColor_Specified, specifiedColor=<rgba 0.0953 0.0953 0.0953 1>}
      uVRegion {bIsValid=false, max=<0 0>, min=<0 0>}
     }
     hoveredForeground {colorUseRule=UseColor_ColorTable, specifiedColor=<rgba 0 0 0 0>}
     hoveredSlateSound {resourceObject=null}
     normal {
      bIsDynamicallyLoaded=false
      drawAs=NoDrawType
      imageType=FullColor
      mirroring=NoMirror
      resourceName=None
      resourceObject=null
      tiling=NoTile
      imageSize=<16 16>
      margin=<ltrb 0 0 0 0>
      outlineSettings {
       bUseBrushTransparency=false
       roundingType=FixedRadius
       width=0
       color {colorUseRule=UseColor_ColorTable, specifiedColor=<rgba 0 0 0 0>}
       cornerRadii=<4 4 4 4>
      }
      tintColor {colorUseRule=UseColor_Specified, specifiedColor=<rgba 0.0176 0.0176 0.0176 1>}
      uVRegion {bIsValid=false, max=<0 0>, min=<0 0>}
     }
     normalForeground {colorUseRule=UseColor_ColorTable, specifiedColor=<rgba 0 0 0 0>}
     normalPadding=<ltrb 6 1.5 6 1.5>
     pressed {
      bIsDynamicallyLoaded=false
      drawAs=RoundedBox
      imageType=NoImage
      mirroring=NoMirror
      resourceName=None
      resourceObject=null
      tiling=NoTile
      imageSize=<32 32>
      margin=<ltrb 0 0 0 0>
      outlineSettings {
       bUseBrushTransparency=false
       roundingType=FixedRadius
       width=0
       color {colorUseRule=UseColor_ColorTable, specifiedColor=<rgba 0 0 0 0>}
       cornerRadii=<4 4 4 4>
      }
      tintColor {colorUseRule=UseColor_Specified, specifiedColor=<rgba 0.0284 0.0284 0.0284 1>}
      uVRegion {bIsValid=false, max=<0 0>, min=<0 0>}
     }
     pressedForeground {colorUseRule=UseColor_ColorTable, specifiedColor=<rgba 0 0 0 0>}
     pressedPadding=<ltrb 6 2.5 6 0.5>
     pressedSlateSound {resourceObject=null}
    }
   }
   slot {class=HorizontalBoxSlot, properties {horizontalAlignment=HAlign_Center}}
  }
  Edit_1 {
   widgetClassPath=&TextBlock
   parent=ClearSearchBox
   widget=Edit_1
   properties {
    autoWrapText=true
    justification=Center
    text=x
    colorAndOpacity {colorUseRule=UseColor_Specified, specifiedColor=<rgba 0.5255 0.5255 0.5255 1>}
    font {
     bForceMonospaced=false
     bMaterialIsStencil=false
     fontMaterial=null
     fontObject=&Roboto
     letterSpacing=0
     monospacedWidth=1
     size=10
     skewAmount=0
     typefaceFontName=Bold
     outlineSettings {
      bApplyOutlineToDropShadows=false
      bMiteredCorners=false
      bSeparateFillAlpha=false
      outlineMaterial=null
      outlineSize=0
      outlineColor=<rgba 0 0 0 1>
     }
    }
   }
   slot {class=ButtonSlot}
  }
  SettingsButton {
   widgetClassPath=&Button
   parent=HorizontalBox_0
   widget=SettingsButton
   properties {
    toolTipText=Settings Menu
    widgetStyle {
     clickedSlateSound {resourceObject=null}
     disabled {
      bIsDynamicallyLoaded=false
      drawAs=RoundedBox
      imageType=NoImage
      mirroring=NoMirror
      resourceName=None
      resourceObject=null
      tiling=NoTile
      imageSize=<0 0>
      margin=<ltrb 0 0 0 0>
      outlineSettings {
       bUseBrushTransparency=false
       roundingType=FixedRadius
       width=1
       color {colorUseRule=UseColor_ColorTable, specifiedColor=<rgba 0 0 0 0>}
       cornerRadii=<4 4 4 4>
      }
      tintColor {colorUseRule=UseColor_Specified, specifiedColor=<rgba 0.0395 0.0395 0.0395 1>}
      uVRegion {bIsValid=false, max=<0 0>, min=<0 0>}
     }
     disabledForeground {colorUseRule=UseColor_ColorTable, specifiedColor=<rgba 0 0 0 0>}
     hovered {
      bIsDynamicallyLoaded=false
      drawAs=RoundedBox
      imageType=FullColor
      mirroring=NoMirror
      resourceName=None
      resourceObject=null
      tiling=NoTile
      imageSize=<28 28>
      margin=<ltrb 0 0 0 0>
      outlineSettings {
       bUseBrushTransparency=false
       roundingType=FixedRadius
       width=0
       color {colorUseRule=UseColor_ColorTable, specifiedColor=<rgba 0 0 0 0>}
       cornerRadii=<4 4 4 4>
      }
      tintColor {colorUseRule=UseColor_Specified, specifiedColor=<rgba 0.0941 0.0941 0.0941 1>}
      uVRegion {bIsValid=false, max=<0 0>, min=<0 0>}
     }
     hoveredForeground {colorUseRule=UseColor_ColorTable, specifiedColor=<rgba 0 0 0 0>}
     hoveredSlateSound {resourceObject=null}
     normal {
      bIsDynamicallyLoaded=false
      drawAs=NoDrawType
      imageType=FullColor
      mirroring=NoMirror
      resourceName=None
      resourceObject=null
      tiling=NoTile
      imageSize=<28 28>
      margin=<ltrb 0 0 0 0>
      outlineSettings {
       bUseBrushTransparency=false
       roundingType=FixedRadius
       width=0
       color {colorUseRule=UseColor_ColorTable, specifiedColor=<rgba 0 0 0 0>}
       cornerRadii=<4 4 4 4>
      }
      tintColor {colorUseRule=UseColor_Specified, specifiedColor=<rgba 0.0176 0.0176 0.0176 1>}
      uVRegion {bIsValid=false, max=<0 0>, min=<0 0>}
     }
     normalForeground {colorUseRule=UseColor_ColorTable, specifiedColor=<rgba 0 0 0 0>}
     normalPadding=<ltrb 0 0 0 0>
     pressed {
      bIsDynamicallyLoaded=false
      drawAs=RoundedBox
      imageType=FullColor
      mirroring=NoMirror
      resourceName=None
      resourceObject=null
      tiling=NoTile
      imageSize=<28 28>
      margin=<ltrb 0 0 0 0>
      outlineSettings {
       bUseBrushTransparency=false
       roundingType=FixedRadius
       width=0
       color {colorUseRule=UseColor_ColorTable, specifiedColor=<rgba 0 0 0 0>}
       cornerRadii=<4 4 4 4>
      }
      tintColor {colorUseRule=UseColor_Specified, specifiedColor=<rgba 0.0392 0.0392 0.0392 1>}
      uVRegion {bIsValid=false, max=<0 0>, min=<0 0>}
     }
     pressedForeground {colorUseRule=UseColor_ColorTable, specifiedColor=<rgba 0 0 0 0>}
     pressedPadding=<ltrb 0 0 0 0>
     pressedSlateSound {resourceObject=null}
    }
   }
   slot {class=HorizontalBoxSlot, properties {horizontalAlignment=HAlign_Center, verticalAlignment=VAlign_Center}}
  }
  HorizontalBox_5 {
   widgetClassPath=&HorizontalBox
   parent=SettingsButton
   widget=HorizontalBox_5
   slot {class=ButtonSlot, properties {padding=<ltrb 0 0 0 0>}}
  }
  AssetImage_2 {
   widgetClassPath=&Image
   bIsVariable=true
   parent=HorizontalBox_5
   widget=AssetImage_2
   properties {brush {
     bIsDynamicallyLoaded=false
     drawAs=Image
     imageType=FullColor
     mirroring=NoMirror
     resourceName=None
     resourceObject=/AnimInEdTools/AnimationBrowser/Icons/Icon_Settings
     tiling=NoTile
     imageSize=<28 28>
     margin=<ltrb 0 0 0 0>
     outlineSettings {
      bUseBrushTransparency=false
      roundingType=HalfHeightRadius
      width=0
      color {colorUseRule=UseColor_Specified, specifiedColor=<rgba 0 0 0 0>}
      cornerRadii=<0 0 0 0>
     }
     tintColor {colorUseRule=UseColor_Specified, specifiedColor=<rgba 1 1 1 1>}
     uVRegion {bIsValid=false, max=<0 0>, min=<0 0>}
    }}
   slot {class=HorizontalBoxSlot, properties {horizontalAlignment=HAlign_Center, verticalAlignment=VAlign_Center}}
  }
  ListHeaderSizeBox {
   widgetClassPath=&SizeBox
   bIsVariable=true
   parent=VerticalBox_0
   widget=ListHeaderSizeBox
   properties {bOverride_MinDesiredHeight=true, minDesiredHeight=20, visibility=Collapsed}
   slot {class=VerticalBoxSlot}
  }
  Border_1 {
   widgetClassPath=&Border
   parent=ListHeaderSizeBox
   widget=Border_1
   properties {verticalAlignment=VAlign_Center, brushColor=<rgba 0.0176 0.0176 0.0176 1>, padding=<ltrb 0 0 0 0>}
   slot {class=SizeBoxSlot, properties {padding=<ltrb 5 2 5 2>}}
  }
  HorizontalBox {
   widgetClassPath=&HorizontalBox
   parent=Border_1
   widget=HorizontalBox
   slot {class=BorderSlot, properties {verticalAlignment=VAlign_Center, padding=<ltrb 0 0 0 0>}}
  }
  RestoreTextBlock {
   widgetClassPath=&TextBlock
   bIsVariable=true
   parent=HorizontalBox
   widget=RestoreTextBlock
   properties {
    justification=Center
    text=Current File
    colorAndOpacity {colorUseRule=UseColor_Specified, specifiedColor=<rgba 0.5255 0.5255 0.5255 1>}
    font {
     bForceMonospaced=false
     bMaterialIsStencil=false
     fontMaterial=null
     fontObject=&Roboto
     letterSpacing=0
     monospacedWidth=1
     size=10
     skewAmount=0
     typefaceFontName=Bold Italic
     outlineSettings {
      bApplyOutlineToDropShadows=false
      bMiteredCorners=false
      bSeparateFillAlpha=false
      outlineMaterial=null
      outlineSize=0
      outlineColor=<rgba 0 0 0 1>
     }
    }
   }
   slot {
    class=HorizontalBoxSlot
    properties {horizontalAlignment=HAlign_Left, verticalAlignment=VAlign_Center, padding=<ltrb 2 0 4 0>, size {sizeRule=Fill, value=1}}
   }
  }
  OptionToolbar {
   widgetClassPath=&EditorUtilityScrollBox
   bIsVariable=true
   parent=HorizontalBox
   widget=OptionToolbar
   properties {
    allowOverscroll=false
    orientation=Orient_Horizontal
    scrollbarPadding=<ltrb 0.5 0.5 0.5 0.1>
    scrollbarThickness=<1 3>
    widgetBarStyle {
     thickness=3
     draggedThumbImage {
      bIsDynamicallyLoaded=false
      drawAs=RoundedBox
      imageType=NoImage
      mirroring=NoMirror
      resourceName=None
      resourceObject=null
      tiling=NoTile
      imageSize=<0 0>
      margin=<ltrb 0 0 0 0>
      outlineSettings {
       bUseBrushTransparency=false
       roundingType=FixedRadius
       width=0
       color {colorUseRule=UseColor_Specified, specifiedColor=<rgba 0 0 0 0>}
       cornerRadii=<4 4 4 4>
      }
      tintColor {colorUseRule=UseColor_Specified, specifiedColor=<rgba 0.2159 0.2159 0.2159 1>}
      uVRegion {bIsValid=false, max=<0 0>, min=<0 0>}
     }
     horizontalBackgroundImage {
      bIsDynamicallyLoaded=false
      drawAs=NoDrawType
      imageType=NoImage
      mirroring=NoMirror
      resourceName=None
      resourceObject=null
      tiling=NoTile
      imageSize=<0 0>
      margin=<ltrb 0 0 0 0>
      outlineSettings {
       bUseBrushTransparency=false
       roundingType=HalfHeightRadius
       width=0
       color {colorUseRule=UseColor_Specified, specifiedColor=<rgba 0 0 0 0>}
       cornerRadii=<0 0 0 0>
      }
      tintColor {colorUseRule=UseColor_Specified, specifiedColor=<rgba 1 1 1 1>}
      uVRegion {bIsValid=false, max=<0 0>, min=<0 0>}
     }
     horizontalBottomSlotImage {
      bIsDynamicallyLoaded=false
      drawAs=NoDrawType
      imageType=NoImage
      mirroring=NoMirror
      resourceName=None
      resourceObject=null
      tiling=NoTile
      imageSize=<0 0>
      margin=<ltrb 0 0 0 0>
      outlineSettings {
       bUseBrushTransparency=false
       roundingType=HalfHeightRadius
       width=0
       color {colorUseRule=UseColor_Specified, specifiedColor=<rgba 0 0 0 0>}
       cornerRadii=<0 0 0 0>
      }
      tintColor {colorUseRule=UseColor_Specified, specifiedColor=<rgba 1 1 1 1>}
      uVRegion {bIsValid=false, max=<0 0>, min=<0 0>}
     }
     horizontalTopSlotImage {
      bIsDynamicallyLoaded=false
      drawAs=NoDrawType
      imageType=NoImage
      mirroring=NoMirror
      resourceName=None
      resourceObject=null
      tiling=NoTile
      imageSize=<0 0>
      margin=<ltrb 0 0 0 0>
      outlineSettings {
       bUseBrushTransparency=false
       roundingType=HalfHeightRadius
       width=0
       color {colorUseRule=UseColor_Specified, specifiedColor=<rgba 0 0 0 0>}
       cornerRadii=<0 0 0 0>
      }
      tintColor {colorUseRule=UseColor_Specified, specifiedColor=<rgba 1 1 1 1>}
      uVRegion {bIsValid=false, max=<0 0>, min=<0 0>}
     }
     hoveredThumbImage {
      bIsDynamicallyLoaded=false
      drawAs=RoundedBox
      imageType=NoImage
      mirroring=NoMirror
      resourceName=None
      resourceObject=null
      tiling=NoTile
      imageSize=<0 0>
      margin=<ltrb 0 0 0 0>
      outlineSettings {
       bUseBrushTransparency=false
       roundingType=FixedRadius
       width=0
       color {colorUseRule=UseColor_Specified, specifiedColor=<rgba 0 0 0 0>}
       cornerRadii=<4 4 4 4>
      }
      tintColor {colorUseRule=UseColor_Specified, specifiedColor=<rgba 0.2159 0.2159 0.2159 1>}
      uVRegion {bIsValid=false, max=<0 0>, min=<0 0>}
     }
     normalThumbImage {
      bIsDynamicallyLoaded=false
      drawAs=RoundedBox
      imageType=NoImage
      mirroring=NoMirror
      resourceName=None
      resourceObject=null
      tiling=NoTile
      imageSize=<0 0>
      margin=<ltrb 0 0 0 0>
      outlineSettings {
       bUseBrushTransparency=false
       roundingType=FixedRadius
       width=0
       color {colorUseRule=UseColor_Specified, specifiedColor=<rgba 0 0 0 0>}
       cornerRadii=<4 4 4 4>
      }
      tintColor {colorUseRule=UseColor_Specified, specifiedColor=<rgba 0.0953 0.0953 0.0953 1>}
      uVRegion {bIsValid=false, max=<0 0>, min=<0 0>}
     }
     verticalBackgroundImage {
      bIsDynamicallyLoaded=false
      drawAs=NoDrawType
      imageType=NoImage
      mirroring=NoMirror
      resourceName=None
      resourceObject=null
      tiling=NoTile
      imageSize=<0 0>
      margin=<ltrb 0 0 0 0>
      outlineSettings {
       bUseBrushTransparency=false
       roundingType=HalfHeightRadius
       width=0
       color {colorUseRule=UseColor_Specified, specifiedColor=<rgba 0 0 0 0>}
       cornerRadii=<0 0 0 0>
      }
      tintColor {colorUseRule=UseColor_Specified, specifiedColor=<rgba 1 1 1 1>}
      uVRegion {bIsValid=false, max=<0 0>, min=<0 0>}
     }
     verticalBottomSlotImage {
      bIsDynamicallyLoaded=false
      drawAs=NoDrawType
      imageType=NoImage
      mirroring=NoMirror
      resourceName=None
      resourceObject=null
      tiling=NoTile
      imageSize=<0 0>
      margin=<ltrb 0 0 0 0>
      outlineSettings {
       bUseBrushTransparency=false
       roundingType=HalfHeightRadius
       width=0
       color {colorUseRule=UseColor_Specified, specifiedColor=<rgba 0 0 0 0>}
       cornerRadii=<0 0 0 0>
      }
      tintColor {colorUseRule=UseColor_Specified, specifiedColor=<rgba 1 1 1 1>}
      uVRegion {bIsValid=false, max=<0 0>, min=<0 0>}
     }
     verticalTopSlotImage {
      bIsDynamicallyLoaded=false
      drawAs=NoDrawType
      imageType=NoImage
      mirroring=NoMirror
      resourceName=None
      resourceObject=null
      tiling=NoTile
      imageSize=<0 0>
      margin=<ltrb 0 0 0 0>
      outlineSettings {
       bUseBrushTransparency=false
       roundingType=HalfHeightRadius
       width=0
       color {colorUseRule=UseColor_Specified, specifiedColor=<rgba 0 0 0 0>}
       cornerRadii=<0 0 0 0>
      }
      tintColor {colorUseRule=UseColor_Specified, specifiedColor=<rgba 1 1 1 1>}
      uVRegion {bIsValid=false, max=<0 0>, min=<0 0>}
     }
    }
    widgetStyle {
     barThickness=3
     bottomShadowBrush {
      bIsDynamicallyLoaded=false
      drawAs=Box
      imageType=FullColor
      mirroring=NoMirror
      resourceName=../../../Engine/Content/Slate/Common/ScrollBoxShadowBottom.png
      resourceObject=null
      tiling=NoTile
      imageSize=<16 8>
      margin=<ltrb 0.5 0 0.5 1>
      outlineSettings {
       bUseBrushTransparency=false
       roundingType=HalfHeightRadius
       width=0
       color {colorUseRule=UseColor_Specified, specifiedColor=<rgba 0 0 0 0>}
       cornerRadii=<0 0 0 0>
      }
      tintColor {colorUseRule=UseColor_Specified, specifiedColor=<rgba 1 1 1 1>}
      uVRegion {bIsValid=false, max=<0 0>, min=<0 0>}
     }
     horizontalScrolledContentPadding=<ltrb 0 0 1 0>
     leftShadowBrush {
      bIsDynamicallyLoaded=false
      drawAs=Box
      imageType=FullColor
      mirroring=NoMirror
      resourceName=../../../Engine/Content/Slate/Common/ScrollBoxShadowLeft.png
      resourceObject=null
      tiling=NoTile
      imageSize=<8 16>
      margin=<ltrb 1 0.5 0 0.5>
      outlineSettings {
       bUseBrushTransparency=false
       roundingType=HalfHeightRadius
       width=0
       color {colorUseRule=UseColor_Specified, specifiedColor=<rgba 0 0 0 0>}
       cornerRadii=<0 0 0 0>
      }
      tintColor {colorUseRule=UseColor_Specified, specifiedColor=<rgba 1 1 1 1>}
      uVRegion {bIsValid=false, max=<0 0>, min=<0 0>}
     }
     rightShadowBrush {
      bIsDynamicallyLoaded=false
      drawAs=Box
      imageType=FullColor
      mirroring=NoMirror
      resourceName=../../../Engine/Content/Slate/Common/ScrollBoxShadowRight.png
      resourceObject=null
      tiling=NoTile
      imageSize=<8 16>
      margin=<ltrb 0 0.5 1 0.5>
      outlineSettings {
       bUseBrushTransparency=false
       roundingType=HalfHeightRadius
       width=0
       color {colorUseRule=UseColor_Specified, specifiedColor=<rgba 0 0 0 0>}
       cornerRadii=<0 0 0 0>
      }
      tintColor {colorUseRule=UseColor_Specified, specifiedColor=<rgba 1 1 1 1>}
      uVRegion {bIsValid=false, max=<0 0>, min=<0 0>}
     }
     topShadowBrush {
      bIsDynamicallyLoaded=false
      drawAs=Box
      imageType=FullColor
      mirroring=NoMirror
      resourceName=../../../Engine/Content/Slate/Common/ScrollBoxShadowTop.png
      resourceObject=null
      tiling=NoTile
      imageSize=<16 8>
      margin=<ltrb 0.5 1 0.5 0>
      outlineSettings {
       bUseBrushTransparency=false
       roundingType=HalfHeightRadius
       width=0
       color {colorUseRule=UseColor_Specified, specifiedColor=<rgba 0 0 0 0>}
       cornerRadii=<0 0 0 0>
      }
      tintColor {colorUseRule=UseColor_Specified, specifiedColor=<rgba 1 1 1 1>}
      uVRegion {bIsValid=false, max=<0 0>, min=<0 0>}
     }
     verticalScrolledContentPadding=<ltrb 0 0 0 1>
    }
   }
   slot {
    class=HorizontalBoxSlot
    properties {horizontalAlignment=HAlign_Right, verticalAlignment=VAlign_Center, padding=<ltrb 0 2 0 2>, size {sizeRule=Fill, value=1}}
   }
  }
  Image_82 {
   widgetClassPath=&Image
   bIsVariable=true
   parent=HorizontalBox
   widget=Image_82
   properties {brush {
     bIsDynamicallyLoaded=false
     drawAs=Image
     imageType=NoImage
     mirroring=NoMirror
     resourceName=None
     resourceObject=null
     tiling=NoTile
     imageSize=<1 18>
     margin=<ltrb 0 0 0 0>
     outlineSettings {
      bUseBrushTransparency=false
      roundingType=HalfHeightRadius
      width=0
      color {colorUseRule=UseColor_Specified, specifiedColor=<rgba 0 0 0 0>}
      cornerRadii=<0 0 0 0>
     }
     tintColor {colorUseRule=UseColor_Specified, specifiedColor=<rgba 0.117 0.117 0.117 1>}
     uVRegion {bIsValid=false, max=<0 0>, min=<0 0>}
    }}
   slot {class=HorizontalBoxSlot, properties {verticalAlignment=VAlign_Center, padding=<ltrb 4 0 2 0>}}
  }
  AddPropButton {
   widgetClassPath=&Button
   parent=HorizontalBox
   widget=AddPropButton
   properties {
    toolTipText=Add Prop to current Level Sequence
    widgetStyle {
     clickedSlateSound {resourceObject=null}
     disabled {
      bIsDynamicallyLoaded=false
      drawAs=RoundedBox
      imageType=NoImage
      mirroring=NoMirror
      resourceName=None
      resourceObject=null
      tiling=NoTile
      imageSize=<16 16>
      margin=<ltrb 0 0 0 0>
      outlineSettings {
       bUseBrushTransparency=false
       roundingType=FixedRadius
       width=1
       color {colorUseRule=UseColor_ColorTable, specifiedColor=<rgba 0 0 0 0>}
       cornerRadii=<4 4 4 4>
      }
      tintColor {colorUseRule=UseColor_Specified, specifiedColor=<rgba 0.6146 0.6146 0.6146 1>}
      uVRegion {bIsValid=false, max=<0 0>, min=<0 0>}
     }
     disabledForeground {colorUseRule=UseColor_ColorTable, specifiedColor=<rgba 0 0 0 0>}
     hovered {
      bIsDynamicallyLoaded=false
      drawAs=RoundedBox
      imageType=NoImage
      mirroring=NoMirror
      resourceName=None
      resourceObject=null
      tiling=NoTile
      imageSize=<16 16>
      margin=<ltrb 0 0 0 0>
      outlineSettings {
       bUseBrushTransparency=false
       roundingType=FixedRadius
       width=0
       color {colorUseRule=UseColor_ColorTable, specifiedColor=<rgba 0 0 0 0>}
       cornerRadii=<4 4 4 4>
      }
      tintColor {colorUseRule=UseColor_Specified, specifiedColor=<rgba 0.0941 0.0941 0.0941 1>}
      uVRegion {bIsValid=false, max=<0 0>, min=<0 0>}
     }
     hoveredForeground {colorUseRule=UseColor_ColorTable, specifiedColor=<rgba 0 0 0 0>}
     hoveredSlateSound {resourceObject=null}
     normal {
      bIsDynamicallyLoaded=false
      drawAs=RoundedBox
      imageType=NoImage
      mirroring=NoMirror
      resourceName=None
      resourceObject=null
      tiling=NoTile
      imageSize=<16 16>
      margin=<ltrb 0 0 0 0>
      outlineSettings {
       bUseBrushTransparency=false
       roundingType=FixedRadius
       width=0
       color {colorUseRule=UseColor_ColorTable, specifiedColor=<rgba 0 0 0 0>}
       cornerRadii=<4 4 4 4>
      }
      tintColor {colorUseRule=UseColor_Specified, specifiedColor=<rgba 0.0208 0.0208 0.0208 0>}
      uVRegion {bIsValid=false, max=<0 0>, min=<0 0>}
     }
     normalForeground {colorUseRule=UseColor_ColorTable, specifiedColor=<rgba 0 0 0 0>}
     normalPadding=<ltrb 0 0 0 0>
     pressed {
      bIsDynamicallyLoaded=false
      drawAs=RoundedBox
      imageType=NoImage
      mirroring=NoMirror
      resourceName=None
      resourceObject=null
      tiling=NoTile
      imageSize=<16 16>
      margin=<ltrb 0 0 0 0>
      outlineSettings {
       bUseBrushTransparency=false
       roundingType=FixedRadius
       width=0
       color {colorUseRule=UseColor_ColorTable, specifiedColor=<rgba 0 0 0 0>}
       cornerRadii=<4 4 4 4>
      }
      tintColor {colorUseRule=UseColor_Specified, specifiedColor=<rgba 0.0275 0.0275 0.0275 1>}
      uVRegion {bIsValid=false, max=<0 0>, min=<0 0>}
     }
     pressedForeground {colorUseRule=UseColor_ColorTable, specifiedColor=<rgba 0 0 0 0>}
     pressedPadding=<ltrb 0 0 0 0>
     pressedSlateSound {resourceObject=null}
    }
   }
   slot {class=HorizontalBoxSlot, properties {horizontalAlignment=HAlign_Right}}
  }
  Image_2 {
   widgetClassPath=&Image
   bIsVariable=true
   parent=AddPropButton
   widget=Image_2
   properties {brush {
     bIsDynamicallyLoaded=false
     drawAs=Image
     imageType=FullColor
     mirroring=NoMirror
     resourceName=None
     resourceObject=/AnimInEdTools/AnimationBrowser/Icons/Icon_AnimationBrowser_AddProp
     tiling=NoTile
     imageSize=<14 14>
     margin=<ltrb 0 0 0 0>
     outlineSettings {
      bUseBrushTransparency=false
      roundingType=HalfHeightRadius
      width=0
      color {colorUseRule=UseColor_Specified, specifiedColor=<rgba 0 0 0 0>}
      cornerRadii=<0 0 0 0>
     }
     tintColor {colorUseRule=UseColor_Specified, specifiedColor=<rgba 0.5271 0.5271 0.5271 1>}
     uVRegion {bIsValid=false, max=<0 0>, min=<0 0>}
    }}
   slot {class=ButtonSlot, properties {padding=<ltrb 3 3 3 3>}}
  }
  AddActorButton {
   widgetClassPath=&Button
   parent=HorizontalBox
   widget=AddActorButton
   properties {
    toolTipText=Add Actor to current Level Sequence
    widgetStyle {
     clickedSlateSound {resourceObject=null}
     disabled {
      bIsDynamicallyLoaded=false
      drawAs=RoundedBox
      imageType=NoImage
      mirroring=NoMirror
      resourceName=None
      resourceObject=null
      tiling=NoTile
      imageSize=<16 16>
      margin=<ltrb 0 0 0 0>
      outlineSettings {
       bUseBrushTransparency=false
       roundingType=FixedRadius
       width=1
       color {colorUseRule=UseColor_ColorTable, specifiedColor=<rgba 0 0 0 0>}
       cornerRadii=<4 4 4 4>
      }
      tintColor {colorUseRule=UseColor_Specified, specifiedColor=<rgba 0.6146 0.6146 0.6146 1>}
      uVRegion {bIsValid=false, max=<0 0>, min=<0 0>}
     }
     disabledForeground {colorUseRule=UseColor_ColorTable, specifiedColor=<rgba 0 0 0 0>}
     hovered {
      bIsDynamicallyLoaded=false
      drawAs=RoundedBox
      imageType=NoImage
      mirroring=NoMirror
      resourceName=None
      resourceObject=null
      tiling=NoTile
      imageSize=<16 16>
      margin=<ltrb 0 0 0 0>
      outlineSettings {
       bUseBrushTransparency=false
       roundingType=FixedRadius
       width=0
       color {colorUseRule=UseColor_ColorTable, specifiedColor=<rgba 0 0 0 0>}
       cornerRadii=<4 4 4 4>
      }
      tintColor {colorUseRule=UseColor_Specified, specifiedColor=<rgba 0.0941 0.0941 0.0941 1>}
      uVRegion {bIsValid=false, max=<0 0>, min=<0 0>}
     }
     hoveredForeground {colorUseRule=UseColor_ColorTable, specifiedColor=<rgba 0 0 0 0>}
     hoveredSlateSound {resourceObject=null}
     normal {
      bIsDynamicallyLoaded=false
      drawAs=RoundedBox
      imageType=NoImage
      mirroring=NoMirror
      resourceName=None
      resourceObject=null
      tiling=NoTile
      imageSize=<16 16>
      margin=<ltrb 0 0 0 0>
      outlineSettings {
       bUseBrushTransparency=false
       roundingType=FixedRadius
       width=0
       color {colorUseRule=UseColor_ColorTable, specifiedColor=<rgba 0 0 0 0>}
       cornerRadii=<4 4 4 4>
      }
      tintColor {colorUseRule=UseColor_Specified, specifiedColor=<rgba 0.0208 0.0208 0.0208 0>}
      uVRegion {bIsValid=false, max=<0 0>, min=<0 0>}
     }
     normalForeground {colorUseRule=UseColor_ColorTable, specifiedColor=<rgba 0 0 0 0>}
     normalPadding=<ltrb 0 0 0 0>
     pressed {
      bIsDynamicallyLoaded=false
      drawAs=RoundedBox
      imageType=NoImage
      mirroring=NoMirror
      resourceName=None
      resourceObject=null
      tiling=NoTile
      imageSize=<16 16>
      margin=<ltrb 0 0 0 0>
      outlineSettings {
       bUseBrushTransparency=false
       roundingType=FixedRadius
       width=0
       color {colorUseRule=UseColor_ColorTable, specifiedColor=<rgba 0 0 0 0>}
       cornerRadii=<4 4 4 4>
      }
      tintColor {colorUseRule=UseColor_Specified, specifiedColor=<rgba 0.0275 0.0275 0.0275 1>}
      uVRegion {bIsValid=false, max=<0 0>, min=<0 0>}
     }
     pressedForeground {colorUseRule=UseColor_ColorTable, specifiedColor=<rgba 0 0 0 0>}
     pressedPadding=<ltrb 0 0 0 0>
     pressedSlateSound {resourceObject=null}
    }
   }
   slot {class=HorizontalBoxSlot, properties {horizontalAlignment=HAlign_Right}}
  }
  Image_1 {
   widgetClassPath=&Image
   bIsVariable=true
   parent=AddActorButton
   widget=Image_1
   properties {brush {
     bIsDynamicallyLoaded=false
     drawAs=Image
     imageType=FullColor
     mirroring=NoMirror
     resourceName=None
     resourceObject=/AnimInEdTools/AnimationBrowser/Icons/Icon_AnimationBrowser_AddCharacter
     tiling=NoTile
     imageSize=<14 14>
     margin=<ltrb 0 0 0 0>
     outlineSettings {
      bUseBrushTransparency=false
      roundingType=HalfHeightRadius
      width=0
      color {colorUseRule=UseColor_Specified, specifiedColor=<rgba 0 0 0 0>}
      cornerRadii=<0 0 0 0>
     }
     tintColor {colorUseRule=UseColor_Specified, specifiedColor=<rgba 0.5271 0.5271 0.5271 1>}
     uVRegion {bIsValid=false, max=<0 0>, min=<0 0>}
    }}
   slot {class=ButtonSlot, properties {padding=<ltrb 3 3 3 3>}}
  }
  RestoreButton {
   widgetClassPath=&Button
   parent=HorizontalBox
   widget=RestoreButton
   properties {
    toolTipText=Restore version of selected asset
    widgetStyle {
     clickedSlateSound {resourceObject=null}
     disabled {
      bIsDynamicallyLoaded=false
      drawAs=RoundedBox
      imageType=NoImage
      mirroring=NoMirror
      resourceName=None
      resourceObject=null
      tiling=NoTile
      imageSize=<16 16>
      margin=<ltrb 0 0 0 0>
      outlineSettings {
       bUseBrushTransparency=false
       roundingType=FixedRadius
       width=1
       color {colorUseRule=UseColor_ColorTable, specifiedColor=<rgba 0 0 0 0>}
       cornerRadii=<4 4 4 4>
      }
      tintColor {colorUseRule=UseColor_Specified, specifiedColor=<rgba 0.6146 0.6146 0.6146 1>}
      uVRegion {bIsValid=false, max=<0 0>, min=<0 0>}
     }
     disabledForeground {colorUseRule=UseColor_ColorTable, specifiedColor=<rgba 0 0 0 0>}
     hovered {
      bIsDynamicallyLoaded=false
      drawAs=RoundedBox
      imageType=NoImage
      mirroring=NoMirror
      resourceName=None
      resourceObject=null
      tiling=NoTile
      imageSize=<16 16>
      margin=<ltrb 0 0 0 0>
      outlineSettings {
       bUseBrushTransparency=false
       roundingType=FixedRadius
       width=0
       color {colorUseRule=UseColor_ColorTable, specifiedColor=<rgba 0 0 0 0>}
       cornerRadii=<4 4 4 4>
      }
      tintColor {colorUseRule=UseColor_Specified, specifiedColor=<rgba 0.0941 0.0941 0.0941 1>}
      uVRegion {bIsValid=false, max=<0 0>, min=<0 0>}
     }
     hoveredForeground {colorUseRule=UseColor_ColorTable, specifiedColor=<rgba 0 0 0 0>}
     hoveredSlateSound {resourceObject=null}
     normal {
      bIsDynamicallyLoaded=false
      drawAs=RoundedBox
      imageType=NoImage
      mirroring=NoMirror
      resourceName=None
      resourceObject=null
      tiling=NoTile
      imageSize=<16 16>
      margin=<ltrb 0 0 0 0>
      outlineSettings {
       bUseBrushTransparency=false
       roundingType=FixedRadius
       width=0
       color {colorUseRule=UseColor_ColorTable, specifiedColor=<rgba 0 0 0 0>}
       cornerRadii=<4 4 4 4>
      }
      tintColor {colorUseRule=UseColor_Specified, specifiedColor=<rgba 0.0208 0.0208 0.0208 0>}
      uVRegion {bIsValid=false, max=<0 0>, min=<0 0>}
     }
     normalForeground {colorUseRule=UseColor_ColorTable, specifiedColor=<rgba 0 0 0 0>}
     normalPadding=<ltrb 0 0 0 0>
     pressed {
      bIsDynamicallyLoaded=false
      drawAs=RoundedBox
      imageType=NoImage
      mirroring=NoMirror
      resourceName=None
      resourceObject=null
      tiling=NoTile
      imageSize=<16 16>
      margin=<ltrb 0 0 0 0>
      outlineSettings {
       bUseBrushTransparency=false
       roundingType=FixedRadius
       width=0
       color {colorUseRule=UseColor_ColorTable, specifiedColor=<rgba 0 0 0 0>}
       cornerRadii=<4 4 4 4>
      }
      tintColor {colorUseRule=UseColor_Specified, specifiedColor=<rgba 0.0275 0.0275 0.0275 1>}
      uVRegion {bIsValid=false, max=<0 0>, min=<0 0>}
     }
     pressedForeground {colorUseRule=UseColor_ColorTable, specifiedColor=<rgba 0 0 0 0>}
     pressedPadding=<ltrb 0 0 0 0>
     pressedSlateSound {resourceObject=null}
    }
   }
   slot {class=HorizontalBoxSlot, properties {horizontalAlignment=HAlign_Right}}
  }
  Image {
   widgetClassPath=&Image
   bIsVariable=true
   parent=RestoreButton
   widget=Image
   properties {brush {
     bIsDynamicallyLoaded=false
     drawAs=Image
     imageType=FullColor
     mirroring=NoMirror
     resourceName=None
     resourceObject=/AnimInEdTools/ContextMenu/Icon/Icon_Time
     tiling=NoTile
     imageSize=<16 16>
     margin=<ltrb 0 0 0 0>
     outlineSettings {
      bUseBrushTransparency=false
      roundingType=HalfHeightRadius
      width=0
      color {colorUseRule=UseColor_Specified, specifiedColor=<rgba 0 0 0 0>}
      cornerRadii=<0 0 0 0>
     }
     tintColor {colorUseRule=UseColor_Specified, specifiedColor=<rgba 1 1 1 1>}
     uVRegion {bIsValid=false, max=<0 0>, min=<0 0>}
    }}
   slot {class=ButtonSlot, properties {padding=<ltrb 3 3 3 3>}}
  }
  ListViewBorder {
   widgetClassPath=&Border
   parent=VerticalBox_0
   widget=ListViewBorder
   properties {brushColor=<rgba 0.0103 0.0103 0.0103 1>, padding=<ltrb 0 0 0 0>}
   slot {class=VerticalBoxSlot, properties {padding=<ltrb 5 2 5 0>, size {sizeRule=Fill, value=1}}}
  }
  ListView {
   widgetClassPath=/Script/UMG.ListView
   bIsVariable=true
   parent=ListViewBorder
   widget=ListView
   properties {
    allowOverscroll=false
    bEnableRightClickScrolling=false
    bReturnFocusToSelection=true
    entryWidgetClass=&/AnimationBrowser_ListItem.AnimationBrowser_ListItem_C
    scrollBarStyle {
     thickness=8
     draggedThumbImage {
      bIsDynamicallyLoaded=false
      drawAs=RoundedBox
      imageType=NoImage
      mirroring=NoMirror
      resourceName=None
      resourceObject=null
      tiling=NoTile
      imageSize=<0 0>
      margin=<ltrb 0 0 0 0>
      outlineSettings {
       bUseBrushTransparency=false
       roundingType=FixedRadius
       width=0
       color {colorUseRule=UseColor_Specified, specifiedColor=<rgba 0 0 0 0>}
       cornerRadii=<4 4 4 4>
      }
      tintColor {colorUseRule=UseColor_Specified, specifiedColor=<rgba 0.2159 0.2159 0.2159 1>}
      uVRegion {bIsValid=false, max=<0 0>, min=<0 0>}
     }
     horizontalBackgroundImage {
      bIsDynamicallyLoaded=false
      drawAs=NoDrawType
      imageType=NoImage
      mirroring=NoMirror
      resourceName=None
      resourceObject=null
      tiling=NoTile
      imageSize=<0 0>
      margin=<ltrb 0 0 0 0>
      outlineSettings {
       bUseBrushTransparency=false
       roundingType=HalfHeightRadius
       width=0
       color {colorUseRule=UseColor_Specified, specifiedColor=<rgba 0 0 0 0>}
       cornerRadii=<0 0 0 0>
      }
      tintColor {colorUseRule=UseColor_Specified, specifiedColor=<rgba 1 1 1 1>}
      uVRegion {bIsValid=false, max=<0 0>, min=<0 0>}
     }
     horizontalBottomSlotImage {
      bIsDynamicallyLoaded=false
      drawAs=NoDrawType
      imageType=NoImage
      mirroring=NoMirror
      resourceName=None
      resourceObject=null
      tiling=NoTile
      imageSize=<0 0>
      margin=<ltrb 0 0 0 0>
      outlineSettings {
       bUseBrushTransparency=false
       roundingType=HalfHeightRadius
       width=0
       color {colorUseRule=UseColor_Specified, specifiedColor=<rgba 0 0 0 0>}
       cornerRadii=<0 0 0 0>
      }
      tintColor {colorUseRule=UseColor_Specified, specifiedColor=<rgba 1 1 1 1>}
      uVRegion {bIsValid=false, max=<0 0>, min=<0 0>}
     }
     horizontalTopSlotImage {
      bIsDynamicallyLoaded=false
      drawAs=NoDrawType
      imageType=NoImage
      mirroring=NoMirror
      resourceName=None
      resourceObject=null
      tiling=NoTile
      imageSize=<0 0>
      margin=<ltrb 0 0 0 0>
      outlineSettings {
       bUseBrushTransparency=false
       roundingType=HalfHeightRadius
       width=0
       color {colorUseRule=UseColor_Specified, specifiedColor=<rgba 0 0 0 0>}
       cornerRadii=<0 0 0 0>
      }
      tintColor {colorUseRule=UseColor_Specified, specifiedColor=<rgba 1 1 1 1>}
      uVRegion {bIsValid=false, max=<0 0>, min=<0 0>}
     }
     hoveredThumbImage {
      bIsDynamicallyLoaded=false
      drawAs=RoundedBox
      imageType=NoImage
      mirroring=NoMirror
      resourceName=None
      resourceObject=null
      tiling=NoTile
      imageSize=<0 0>
      margin=<ltrb 0 0 0 0>
      outlineSettings {
       bUseBrushTransparency=false
       roundingType=FixedRadius
       width=0
       color {colorUseRule=UseColor_Specified, specifiedColor=<rgba 0 0 0 0>}
       cornerRadii=<4 4 4 4>
      }
      tintColor {colorUseRule=UseColor_Specified, specifiedColor=<rgba 0.2159 0.2159 0.2159 1>}
      uVRegion {bIsValid=false, max=<0 0>, min=<0 0>}
     }
     normalThumbImage {
      bIsDynamicallyLoaded=false
      drawAs=RoundedBox
      imageType=NoImage
      mirroring=NoMirror
      resourceName=None
      resourceObject=null
      tiling=NoTile
      imageSize=<0 0>
      margin=<ltrb 0 0 0 0>
      outlineSettings {
       bUseBrushTransparency=false
       roundingType=FixedRadius
       width=0
       color {colorUseRule=UseColor_Specified, specifiedColor=<rgba 0 0 0 0>}
       cornerRadii=<4 4 4 4>
      }
      tintColor {colorUseRule=UseColor_Specified, specifiedColor=<rgba 0.0953 0.0953 0.0953 1>}
      uVRegion {bIsValid=false, max=<0 0>, min=<0 0>}
     }
     verticalBackgroundImage {
      bIsDynamicallyLoaded=false
      drawAs=NoDrawType
      imageType=NoImage
      mirroring=NoMirror
      resourceName=None
      resourceObject=null
      tiling=NoTile
      imageSize=<0 0>
      margin=<ltrb 0 0 0 0>
      outlineSettings {
       bUseBrushTransparency=false
       roundingType=HalfHeightRadius
       width=0
       color {colorUseRule=UseColor_Specified, specifiedColor=<rgba 0 0 0 0>}
       cornerRadii=<0 0 0 0>
      }
      tintColor {colorUseRule=UseColor_Specified, specifiedColor=<rgba 1 1 1 1>}
      uVRegion {bIsValid=false, max=<0 0>, min=<0 0>}
     }
     verticalBottomSlotImage {
      bIsDynamicallyLoaded=false
      drawAs=NoDrawType
      imageType=NoImage
      mirroring=NoMirror
      resourceName=None
      resourceObject=null
      tiling=NoTile
      imageSize=<0 0>
      margin=<ltrb 0 0 0 0>
      outlineSettings {
       bUseBrushTransparency=false
       roundingType=HalfHeightRadius
       width=0
       color {colorUseRule=UseColor_Specified, specifiedColor=<rgba 0 0 0 0>}
       cornerRadii=<0 0 0 0>
      }
      tintColor {colorUseRule=UseColor_Specified, specifiedColor=<rgba 1 1 1 1>}
      uVRegion {bIsValid=false, max=<0 0>, min=<0 0>}
     }
     verticalTopSlotImage {
      bIsDynamicallyLoaded=false
      drawAs=NoDrawType
      imageType=NoImage
      mirroring=NoMirror
      resourceName=None
      resourceObject=null
      tiling=NoTile
      imageSize=<0 0>
      margin=<ltrb 0 0 0 0>
      outlineSettings {
       bUseBrushTransparency=false
       roundingType=HalfHeightRadius
       width=0
       color {colorUseRule=UseColor_Specified, specifiedColor=<rgba 0 0 0 0>}
       cornerRadii=<0 0 0 0>
      }
      tintColor {colorUseRule=UseColor_Specified, specifiedColor=<rgba 1 1 1 1>}
      uVRegion {bIsValid=false, max=<0 0>, min=<0 0>}
     }
    }
    widgetStyle {backgroundBrush {
      bIsDynamicallyLoaded=false
      drawAs=Image
      imageType=NoImage
      mirroring=NoMirror
      resourceName=None
      resourceObject=null
      tiling=NoTile
      imageSize=<0 0>
      margin=<ltrb 0 0 0 0>
      outlineSettings {
       bUseBrushTransparency=false
       roundingType=HalfHeightRadius
       width=0
       color {colorUseRule=UseColor_Specified, specifiedColor=<rgba 0 0 0 0>}
       cornerRadii=<0 0 0 0>
      }
      tintColor {colorUseRule=UseColor_Specified, specifiedColor=<rgba 0.0103 0.0103 0.0103 1>}
      uVRegion {bIsValid=false, max=<0 0>, min=<0 0>}
     }}
   }
   slot {class=BorderSlot, properties {padding=<ltrb 0 0 0 0>}}
  }
  FooterSizeBox {
   widgetClassPath=&SizeBox
   parent=VerticalBox_0
   widget=FooterSizeBox
   properties {bOverride_MinDesiredHeight=true, minDesiredHeight=20}
   slot {class=VerticalBoxSlot}
  }
  Border {
   widgetClassPath=&Border
   parent=FooterSizeBox
   widget=Border
   properties {verticalAlignment=VAlign_Center, brushColor=<rgba 0.0176 0.0176 0.0176 1>, padding=<ltrb 0 0 0 0>}
   slot {class=SizeBoxSlot, properties {padding=<ltrb 5 2 5 2>}}
  }
  HorizontalBox_2 {
   widgetClassPath=&HorizontalBox
   parent=Border
   widget=HorizontalBox_2
   slot {class=BorderSlot, properties {verticalAlignment=VAlign_Center, padding=<ltrb 0 0 0 0>}}
  }
  NumFilesTextBlock {
   widgetClassPath=&TextBlock
   bIsVariable=true
   parent=HorizontalBox_2
   widget=NumFilesTextBlock
   properties {
    justification=Center
    colorAndOpacity {colorUseRule=UseColor_Specified, specifiedColor=<rgba 0.147 0.147 0.147 1>}
    font {
     bForceMonospaced=false
     bMaterialIsStencil=false
     fontMaterial=null
     fontObject=&Roboto
     letterSpacing=0
     monospacedWidth=1
     size=10
     skewAmount=0
     typefaceFontName=Regular
     outlineSettings {
      bApplyOutlineToDropShadows=false
      bMiteredCorners=false
      bSeparateFillAlpha=false
      outlineMaterial=null
      outlineSize=0
      outlineColor=<rgba 0 0 0 1>
     }
    }
   }
   slot {
    class=HorizontalBoxSlot
    properties {horizontalAlignment=HAlign_Right, verticalAlignment=VAlign_Center, padding=<ltrb 4 2 4 2>, size {sizeRule=Fill, value=1}}
   }
  }
  Border_1202 {
   widgetClassPath=&Border
   parent=CanvasPanel
   widget=Border_1202
   properties {brushColor=<rgba 0.0176 0.0176 0.0176 1>, padding=<ltrb 0 0 0 0>}
   slot {class=CanvasPanelSlot, properties {layoutData {alignment=<0 0>, anchors {maximum=<1 1>, minimum=<0 0>}, offsets=<ltrb 0 0 0 0>}}}
  }
  Overlay_67 {
   widgetClassPath=&Overlay
   parent=CanvasPanel
   widget=Overlay_67
   slot {class=CanvasPanelSlot, properties {zOrder=1, layoutData {alignment=<0 0>, anchors {maximum=<1 1>, minimum=<0 0>}, offsets=<ltrb 0 0 0 0>}}}
  }
  ModalLayer {
   widgetClassPath=&NamedSlot
   bIsVariable=true
   parent=Overlay_67
   widget=ModalLayer
   properties {bExposeOnInstanceOnly=true}
   slot {class=OverlaySlot, properties {horizontalAlignment=HAlign_Fill, verticalAlignment=VAlign_Fill}}
  }
  EUW_ContextMenu {
   widgetClassPath=/AnimInEdTools/ContextMenu/EUW_ContextMenu
   bIsVariable=true
   parent=Overlay_67
   widget=EUW_ContextMenu
   properties {visibility=Collapsed}
   slot {class=OverlaySlot}
  }
 ]
}
