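# Editor utility list-row widget for the Animation Browser, showing one anim sequence with its subfolder, Level Sequence link icon and an AnimMontage button.
# Setting the item validates the asset path, finds referencing montages via the Asset Registry, then async-loads the sequence to resolve its Level Sequence link.
# A randomly delayed async source-control query then restyles the row by file state.
# Selection, F2, Delete and right-click are handled, mostly through the parent browser.
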EditorUtilityWidgetBlueprint /AnimInEdTools/AnimationBrowser/AnimationBrowser_ListItem
graphs [
 Asyn P4 Update {nodes [
   n1 = Asyn P4 Update!() -> (then, FileStateOut:Source Control State Structure) {is=title K2Node_FunctionEntry}
   > n2 = Sequence!(execute) -> (then_0, then_1, then_2:Exec, then_3:Exec, then_4:Exec, then_5:Exec)
   n3 = Break Source Control State!(SourceControlState:"Source Control State Structure (by ref)"=@n1) -> (
    Filename:String
    bIsValid:bool=~
    bIsUnknown:bool=~
    bCanCheckIn:bool=~
    bCanCheckOut:bool=~
    bIsCheckedOut:bool=~
    bIsCurrent:bool=~
    bIsSourceControlled:bool=~
    bIsAdded:bool=~
    bIsDeleted:bool=~
    bIsIgnored:bool=~
    bCanEdit:bool=~
    bCanDelete:bool=~
    bIsModified:bool=~
    bCanAdd:bool=~
    bIsConflicted:bool=~
    bCanRevert:bool=~
    bIsCheckedOutOther:bool=~
    CheckedOutOther:String
    bIsCheckedOutInOtherBranch:bool=~
    bIsModifiedInOtherBranch:bool=~
    PreviousUser:String
   ) {is=title K2Node_BreakStruct}
   n4 = Branch(execute=@n2.then_1, Condition=@n3.bIsCurrent)
   Set Style!(
    execute=@n4.else
    self
    Extra Text Color:Linear Color Structure=(R=0.7135,G=0.646,B=0,A=1)
    Extra Text Typeface:Name=Bold
    Icon:Object Reference=/AnimInEdTools/AnimationBrowser/Icons/Icon_LevelSequence_alert
    Extra Text:String
    Main Text Color:Linear Color Structure=(R=0.5255,G=0.5255,B=0.5255,A=1)
    Icon Tint:Slate Color Structure=Make Slate Color(SpecifiedColor=(R=0.7137,G=0.6471,B=0,A=1))
    Tool Tip Text:String=There is a newer version of this asset checked in
   ) -> (then) {is=title K2Node_CallFunction}
   Branch(execute=@n2.then_2, Condition=@n3.bIsAdded)
   > Set Style(
    Extra Text Color=(R=0,G=0.338,B=0.5255,A=1)
    Extra Text Typeface=Bold
    Icon=/AnimInEdTools/AnimationBrowser/Icons/Icon_LevelSequence_add
    Main Text Color=(R=0.5255,G=0.5255,B=0.5255,A=1)
    Icon Tint=Make Slate Color(SpecifiedColor=(R=0.0314,G=0.5255,B=0.0627,A=1))
    Tool Tip Text=The file is marked for add by you
   )
   n5 = Branch(execute=@n2.then_3, Condition=@n3.bIsSourceControlled)
   Set Style(
    execute=@n5.else
    Extra Text Color=(R=0.1714,G=0.1714,B=0.1714,A=1)
    Extra Text Typeface=Regular
    Icon=/AnimInEdTools/AnimationBrowser/Icons/Icon_LevelSequence_not_in_source
    Main Text Color=(R=0.1714,G=0.1714,B=0.1714,A=1)
    Icon Tint=Make Slate Color(SpecifiedColor=(R=0.7137,G=0.6471,B=0,A=1))
    Tool Tip Text=Local File
   )
   Branch(execute=@n2.then_4, Condition=@n3.bIsCheckedOut)
   > Set Style(
    Extra Text Color=(R=0.0319,G=0.5271,B=0.063,A=1)
    Extra Text Typeface=Bold
    Icon=/AnimInEdTools/AnimationBrowser/Icons/Icon_AnimationBrowser_Red_Checkmark
    Main Text Color=(R=0.5255,G=0.5255,B=0.5255,A=1)
    Icon Tint=Make Slate Color(SpecifiedColor=(R=0.0314,G=0.5255,B=0.0627,A=1))
    Tool Tip Text=The file is checked out
   )
   Branch(execute=@n2.then_5, Condition=@n3.bIsCheckedOutOther)
   n6 = Split(SourceString=@n3.CheckedOutOther, InStr=" @", SearchDir=FromStart)
   > Set Style(
    Extra Text Color=(R=0.5271,G=0,B=0,A=1)
    Extra Text Typeface=Bold
    Icon=&Icon_AnimationBrowser_Red_Checkmark
    Main Text Color=(R=0.5255,G=0.5255,B=0.5255,A=1)
    Icon Tint=Make Slate Color(SpecifiedColor=(R=0.5255,G=0,B=0,A=1))
    Tool Tip Text=Append(A="The file is currently checked out by: ", B=@n6)
   )
   Append!(A="(Checked out by: ", B=@n6, C:String=")") -> (ReturnValue)
  ]}
 Display Selected {nodes [
   n1 = Display Selected!() -> (then, Index:bool=~) {is=title K2Node_FunctionEntry}
   > SetBrushColor(
    self=Get Border() -> (Border:Border Object Reference)
    InBrushColor=SelectColor(A=(R=0,G=0.1608,B=0.7451,A=1), B=(R=0,G=0,B=0,A=0), bPickA=@n1)
   )
   > SetColorAndOpacity(
    self=Get TextBlock() -> (TextBlock:Text Object Reference)
    InColorAndOpacity=Select(
     Option 0:Slate Color Structure=Get Color And Opacity() -> (Color And Opacity:Slate Color Structure)
     Option 1:Slate Color Structure=Make Slate Color(SpecifiedColor=(R=0.0103,G=0.0103,B=0.0103,A=1))
     Index=@n1
    ) -> (ReturnValue:Slate Color Structure)
   )
  ]}
 EventGraph {nodes [
   "On Clicked (AnimMontageButton)"() {is=title K2Node_ComponentBoundEvent}
   n1 = Get AnimMontages!(self:Object Animation Browser Object Reference=Get Item Data() -> (Item Data:Object Animation Browser Object Reference)) -> (AnimMontages:Array of Soft Object References)
   > n2 = Switch on Int!(execute:Exec, Selection:int=Length(TargetArray:Array of Soft Object References=@n1)) -> (Default:Exec, 0:Exec, 1:Exec)
   n3 = LoadAsset_Blocking(execute=@n2.1, Asset="Get (a copy)"(Array:Array of Soft Object References=@n1) -> (Output:Soft Object Reference))
   > OpenEditorForAssets!(
    execute
    self:Asset Editor Subsystem Object Reference=Get AssetEditorSubsystem()
    Assets:Array of Object References=Make Array("[0]":Object Reference=@n3) -> (Array:Array of Object References)
    OpenedMethod:EAssetTypeActivationOpenedMethod Enum=Edit
   ) -> (then, ReturnValue:bool)
   Display Anim Montage Menu!(
    execute=@n2.Default
    self:Animation Browser Object Reference=Get AnimationBrowser!() -> (Animation Browser:Animation Browser Object Reference) {is=title K2Node_CallFunction}
    List View Object:Object Animation Browser Object Reference=Get Item Data()
   ) -> (then)
   "On Hovered (AnimMontageButton)"() {is=title K2Node_ComponentBoundEvent}
   > Set ColorAndOpacity!(
    execute
    ColorAndOpacity:Slate Color Structure=Make Slate Color(SpecifiedColor=(R=0.2935,G=0.2938,B=0.3125,A=1))
    self:Text Object Reference=Get MontageTextBlock() -> (MontageTextBlock:Text Object Reference)
   ) -> (then, Output_Get:Slate Color Structure)
   "On Unhovered (AnimMontageButton)"() {is=title K2Node_ComponentBoundEvent}
   > Set ColorAndOpacity(ColorAndOpacity=Make Slate Color(SpecifiedColor=(R=0.0471,G=0.0667,B=0.4,A=1)), self=Get MontageTextBlock())
   Reload Asset() {is=title K2Node_CustomEvent}
   > n4 = Sequence!(execute) -> (then_0, then_1, then_2)
   Display Selected!(execute=[@n22.then_0, @n4.then_0], self, Index:bool=false) -> (then) {is=title K2Node_CallFunction}
   n5 = Event OnListItemObjectSet()
   n6 = Cast To Object_AnimationBrowser!(execute=[@n22.then_1, @n4.then_1], Object=@n5.ListItemObject) -> (then, CastFailed, AsObject Animation Browser:Object Animation Browser Object Reference) {is=title K2Node_DynamicCast}
   > n7 = Set Item Data(Item Data:Object Animation Browser Object Reference=@n6) -> (Output_Get:Object Animation Browser Object Reference)
   SetVisibility(
    execute=[@n6.CastFailed, @n7.then]
    self=Get SubFolderBorder() -> (SubFolderBorder:Editor Utility Button Object Reference)
    InVisibility=Select(
     Option 0:ESlateVisibility Enum=Collapsed
     Option 1:ESlateVisibility Enum=Visible
     Index=Get Display Subfolder!(self:Object Animation Browser Object Reference=@n7) -> (Display Subfolder:bool)
    ) -> (ReturnValue:ESlateVisibility Enum)
   )
   n8 = To Soft Object Path!(SoftObjectReference:Soft Object Reference=Get!(self:Object Animation Browser Object Reference=Get Item Data()) -> (
     Data_AnimSequence_31_#1:Soft Object Reference
     Data_LevelSequence_32_#2:Soft Object Reference
     Data_CurrentUser_26_#3:String
    )) -> (ReturnValue_PathString:String)
   > Validate Anim Sequence Exists!(execute, self, Path:String=@n8) -> (then)
   n9 = GetBaseFilename(InPath=@n8, bRemovePath=true)
   n10 = Split(SourceString=@n9, InStr=_, SearchDir=FromStart)
   n11 = Get Filter Match Asset Name!(self:SG Animation Browser Object Reference=Get Save Game!(self:Animation Browser Object Reference=Get AnimationBrowser()) -> (Save Game:SG Animation Browser Object Reference)) -> (Filter Match Asset Name:bool)
   > Set Text(Text="To Text (String)"(InString=SelectString(A=@n10.RightS, B=@n9, bPickA=@n11)), self=Get TextBlock())
   n12 = GetBaseFilename(InPath=GetPath!(InPath:String=@n8) -> (ReturnValue:String), bRemovePath=true)
   > Set Text(Text="To Text (String)"(InString=@n12), self=Get SubFolderTextBlock() -> (SubFolderTextBlock:Text Object Reference))
   > SetBrush(
    self=Get Icon() -> (Icon:Image Object Reference)
    InBrush=Make Slate Brush!(
     TintColor:Slate Color Structure=(SpecifiedColor=(R=1,G=1,B=1,A=1),ColorUseRule=UseColor_Specified)
     DrawAs:ESlateBrushDrawType Enum=Image
     Tiling:ESlateBrushTileType Enum=NoTile
     Mirroring:ESlateBrushMirrorType Enum=NoMirror
     ImageSize:Vector 2D Structure=(X=14,Y=14)
     Margin:Margin Structure=(Left=0,Top=0,Right=0,Bottom=0)
     ResourceObject:Object Reference=/AnimInEdTools/AnimationBrowser/Icons/Icon_AnimationBrowser_BrokenLink
     OutlineSettings:Slate Brush Outline Settings Structure=(CornerRadii=(X=0,Y=0,Z=0,W=0),Color=(SpecifiedColor=(R=0,G=0,B=0,A=0),ColorUseRule=UseColor_Specified),Width=0,RoundingType=HalfHeightRadius,bUseBrushTransparency=False)
    ) -> (SlateBrush:Slate Brush Structure) {is=title K2Node_MakeStruct}
   )
   > Find and Set Linked AnimMontages()
   > Set ToolTipText!(execute, ToolTipText:Text=MakeLiteralText(Value="LOCTEXT(\"Cannot find a linked Level Sequence\")"), self:Widget Object Reference=Get Icon()) -> (then, Output_Get:Text)
   n13 = To Soft Object Path(SoftObjectReference=Get(self=Get Item Data()))
   n14 = Async Load Asset(execute=[@n22.then_2, @n4.then_2], Asset=To Soft Object Reference(SoftObjectPath=MakeSoftClassPath(PathString=@n13)))
   n15 = Cast To AnimSequence!(execute=@n14.Completed, Object=@n14) -> (then, CastFailed, AsAnim Sequence:Anim Sequence Object Reference)
   > n16 = GetLevelSequenceLinkFromAnimSequence!(execute, InAnimSequence:Anim Sequence Object Reference=@n15) -> (then, ReturnValue:Anim Sequence Level Sequence Link Object Reference)
   > Is Valid(InputObject=@n16)
   n17 = Get Item Data()
   n18 = Get(self=@n17)
   > Set!(
    execute
    Data_AnimSequence_31_#1:Soft Object Reference=@n18
    Data_LevelSequence_32_#2:Soft Object Reference=To Soft Object Reference!(SoftObjectPath_PathString:String=Get!(self:Anim Sequence Level Sequence Link Object Reference=@n16) -> (PathToLevelSequence_PathString:String)) -> (ReturnValue)
    Data_CurrentUser_26_#3:String=@n18.Data_CurrentUser_26_#3
    self:Object Animation Browser Object Reference=@n17
   ) -> (then, Output_Get:Struct Animation Brower Item Structure)
   > SetBrush(
    self=Get Icon()
    InBrush=Make Slate Brush(
     TintColor=(SpecifiedColor=(R=1,G=1,B=1,A=1),ColorUseRule=UseColor_Specified)
     DrawAs=Image
     Tiling=NoTile
     Mirroring=NoMirror
     ImageSize=(X=14,Y=14)
     Margin=(Left=0,Top=0,Right=0,Bottom=0)
     ResourceObject=/AnimInEdTools/AnimationBrowser/Icons/Icon_LevelSequence
     OutlineSettings=(CornerRadii=(X=0,Y=0,Z=0,W=0),Color=(SpecifiedColor=(R=0,G=0,B=0,A=0),ColorUseRule=UseColor_Specified),Width=0,RoundingType=HalfHeightRadius,bUseBrushTransparency=False)
    )
   )
   > RetriggerableDelay!(
    execute
    WorldContextObject:Object Reference
    Duration=MapRangeClamped(Value=RandomFloat!() -> (ReturnValue:double), InRangeA=0, InRangeB=1, OutRangeA=0.5, OutRangeB=0.7)
   ) -> (then)
   n19 = Get!(self:Object Animation Browser Object Reference=Get Item Data()) -> (
    Data_AnimSequence_31_#1:Soft Object Reference
    Data_LevelSequence_32_#2:Soft Object Reference
    Data_CurrentUser_26_#3:String
   )
   > AsyncQueryFileState!(
    execute
    FileStateCallback:Delegate=Create Event()
    InFile:String=GetBaseFilename(InPath=To Soft Object Path(SoftObjectReference=@n19), bRemovePath=false)
    bSilent:bool=true
   ) -> (then)
   Event Construct()
   > Set Color And Opacity(Color And Opacity:Slate Color Structure=Get ColorAndOpacity!(self:Text Object Reference=Get TextBlock()) -> (ColorAndOpacity:Slate Color Structure)) -> (Output_Get:Slate Color Structure)
   n20 = Event On Item Selection Changed!() -> (OutputDelegate, then, bIsSelected:bool) {is=title K2Node_Event}
   > Display Selected(Index=@n20.bIsSelected)
   > Branch(Condition=@n20.bIsSelected)
   > Set Current List Item!(
    execute
    Current List Item:Animation Browser List Item Object Reference=Self-Reference()
    self:Animation Browser Object Reference=Get AnimationBrowser()
   ) -> (then, Output_Get:Animation Browser List Item Object Reference)
   Append!(A, B=" (", C=@n12, D:String=")") -> (ReturnValue)
   n21 = GetPath(InPath=@n8)
   Make Array!("[0]":String=GetBaseFilename(InPath=GetPath(InPath=@n21), bRemovePath=true), "[1]":String=@n12, "[2]":String=@n9) -> (Array:Array of Strings)
   n22 = Sequence(execute=@n5.then)
  ]}
 Find and Set Linked AnimMontages {nodes [
   Find and Set Linked AnimMontages() {is=title K2Node_FunctionEntry}
   > Clear(TargetArray:Array of Soft Object References=Get AnimMontages!() -> (AnimMontages))
   n1 = To Soft Object Path(SoftObjectReference=Get(self=Get Item Data()))
   > n2 = Get Referencers!(
    execute
    self:Asset Registry Interface=GetAssetRegistry()
    PackageName:Name=String To Name(InString=GetBaseFilename(InPath=@n1, bRemovePath=false))
    ReferenceOptions:"Asset Registry Dependency Options Structure (by ref)"=Make Asset Registry Dependency Options(bIncludeSoftPackageReferences=false)
   ) -> (then, OutReferencers:Array of Names, ReturnValue:bool) {is=title K2Node_CallFunction}
   > n3 = For Each Loop(Array:Array of Names=@n2) -> (Array Element:Name)
   Set AnimMontages!(
    execute=@n3.Completed
    AnimMontages:Array of Soft Object References=Get AnimMontages()
    self:Object Animation Browser Object Reference=Get Item Data()
   ) -> (then, Output_Get:Array of Soft Object References)
   n4 = Length(TargetArray=Get AnimMontages())
   > SetVisibility(
    self=Get AnimMontageButton() -> (AnimMontageButton:Button Object Reference)
    InVisibility=Select(Option 0=Collapsed, Option 1=Visible, Index="integer > integer"(A=@n4))
   )
   n5 = FindAssetData!(execute=@n3.LoopBody, AssetPath:String="To String (Name)"(InName=@n3)) -> (then, ReturnValue:Asset Data Structure)
   n6 = Break Asset Data(AssetData=@n5)
   > Branch(Condition="Equal Exactly (String)"(
     A=BreakTopLevelAssetPath!(TopLevelAssetPath:"Top Level Asset Path Structure (by ref)"=@n6.AssetClassPath) -> (PathString:String)
     B=/Script/Engine.AnimMontage
    ))
   > Add(
    TargetArray:Array of Soft Object References=Get AnimMontages()
    NewItem:"Soft Object Reference (by ref)"=To Soft Object Reference!(SoftObjectPath:"Soft Object Path Structure (by ref)"=ToSoftObjectPath!(InAssetData:"Asset Data Structure (by ref)"=@n5) -> (ReturnValue:Soft Object Path Structure)) -> (ReturnValue)
   ) {is=title K2Node_CallArrayFunction}
  ]}
 Get AnimationBrowser {nodes [
   Get AnimationBrowser() {is=title K2Node_FunctionEntry}
   > n1 = FindParentWidgetOfType!(
    execute
    StartingWidget:Widget Object Reference=GetOwningListView!(UserListEntry:User List Entry Interface) -> (ReturnValue:List View Base Object Reference)
    Type:Widget Class Reference=&/Animation_Browser.Animation_Browser_C
   ) -> (then, ReturnValue:Animation Browser Object Reference)
   > Return Node!(execute, Animation Browser:Animation Browser Object Reference=@n1)
  ]}
 OnKeyDown {nodes [
   n1 = OnKeyDown!() -> (then, MyGeometry:Geometry Structure, InKeyEvent:Key Event Structure) {is=title K2Node_FunctionEntry}
   > n2 = Branch(Condition="Equal (Key)"(A:Key Structure=GetKey!(Input:"Key Event Structure (by ref)"=@n1.InKeyEvent) -> (ReturnValue:Key Structure), B:Key Structure=F2) -> (ReturnValue:bool=~) {is=title K2Node_PromotableOperator})
   > Return Node!(execute, ReturnValue:Event Reply Structure=Handled())
   n3 = Branch(execute=@n2.else, Condition="Equal (Key)"(A=GetKey(Input=@n1.InKeyEvent), B=Delete) -> (ReturnValue=~))
   > Delete Item!(
    execute
    self:Animation Browser Object Reference=Get AnimationBrowser() {is=title K2Node_CallFunction}
    Data:Object Animation Browser Object Reference=Get Item Data()
   ) -> (then)
   > Return Node(ReturnValue=Handled())
   Return Node(execute=@n3.else, ReturnValue=Unhandled())
  ]}
 OnMouseButtonDown {nodes [
   n1 = OnMouseButtonDown!() -> (then, MyGeometry:Geometry Structure, MouseEvent:"Pointer Event Structure (by ref)") {is=title K2Node_FunctionEntry}
   > n2 = Branch(Condition=Is Mouse Button Down!(Input:"Pointer Event Structure (by ref)"=@n1.MouseEvent, MouseButton:Key Structure=RightMouseButton) -> (ReturnValue:bool))
   > n3 = FindParentWidgetOfType(StartingWidget=Self-Reference(), Type=&Animation_Browser_C)
   > "On Item Right Clicked (List View)"!(
    execute
    self:Animation Browser Object Reference=@n3
    Item:Object Animation Browser Object Reference=Get Item Data()
    Mouse Event:Pointer Event Structure=Get MouseEvent() -> (MouseEvent:"Pointer Event Structure (by ref)")
   ) -> (then)
   > Return Node(ReturnValue=Handled())
   Return Node(execute=@n2.else, ReturnValue=Unhandled())
  ]}
 Set Style {nodes [
   n1 = Get SubFolderTextBlock()
   SetColorAndOpacity(self=@n1, InColorAndOpacity=Make Slate Color(SpecifiedColor=Get Extra Text Color() -> (Extra Text Color:Linear Color Structure)))
   n2 = Get!(self:Text Object Reference=@n1) -> (
    Font_FontObject:Object Reference
    Font_FontMaterial:Object Reference
    Font_OutlineSettings:Font Outline Settings Structure
    Font_TypefaceFontName:Name
    Font_Size:float=~
    Font_LetterSpacing:int
    Font_SkewAmount:float=~
    Font_bForceMonospaced:bool=~
    Font_bMaterialIsStencil:bool=~
    Font_MonospacedWidth:float=~
   )
   > Set!(
    execute
    Font_FontObject:Object Reference=@n2
    Font_FontMaterial:Object Reference=@n2.Font_FontMaterial
    Font_OutlineSettings:Font Outline Settings Structure=@n2.Font_OutlineSettings
    Font_TypefaceFontName:Name=Get Extra Text Typeface() -> (Extra Text Typeface:Name=None)
    Font_Size:float=@n2.Font_Size
    Font_LetterSpacing:int=@n2.Font_LetterSpacing
    Font_SkewAmount:float=@n2.Font_SkewAmount
    Font_bForceMonospaced:bool=false
    Font_bMaterialIsStencil:bool=false
    Font_MonospacedWidth:float=1
    self:Text Object Reference=@n1
   ) -> (then, Output_Get:Slate Font Info Structure)
   n3 = Set Style!() -> (
    then
    Extra Text Color:Linear Color Structure
    Extra Text Typeface:Name=Regular
    Icon:Object Reference
    Extra Text:String
    Main Text Color:Linear Color Structure=(R=0.5255,G=0.5255,B=0.5255,A=1)
    Icon Tint:Slate Color Structure
    Tool Tip Text:String
   ) {is=title K2Node_FunctionEntry}
   > n4 = Sequence!(execute) -> (then_0, then_1, then_2, then_3:Exec, then_4:Exec)
   SetColorAndOpacity(
    execute=@n4.then_1
    self=Get TextBlock()
    InColorAndOpacity=Make Slate Color(SpecifiedColor=Get Main Text Color() -> (Main Text Color:Linear Color Structure))
   )
   SetBrush(
    execute=@n4.then_2
    self=Get Icon()
    InBrush=Make Slate Brush(
     TintColor=@n3.Icon Tint
     DrawAs=Image
     Tiling=NoTile
     Mirroring=NoMirror
     ImageSize=(X=14,Y=14)
     Margin=(Left=0,Top=0,Right=0,Bottom=0)
     ResourceObject=@n3.Icon
     OutlineSettings=(CornerRadii=(X=0,Y=0,Z=0,W=0),Color=(SpecifiedColor=(R=0,G=0,B=0,A=0),ColorUseRule=UseColor_Specified),Width=0,RoundingType=HalfHeightRadius,bUseBrushTransparency=False)
    )
   )
   n5 = Branch(execute=@n4.then_3, Condition=IsEmpty(InString=@n3.Extra Text))
   n6 = Get SubFolderTextBlock()
   n7 = "To String (Text)"(InText=Get Text(self:Text Object Reference=@n6))
   Set Text(execute=@n5.else, Text="To Text (String)"(InString=Append!(A=@n7, B=" ", C=@n3.Extra Text) -> (ReturnValue)), self=@n6)
   Set ToolTipText(execute=@n4.then_4, ToolTipText="To Text (String)"(InString=Get Tool Tip Text() -> (Tool Tip Text:String)), self=Get Icon())
  ]}
 Validate Anim Sequence Exists {nodes [
   n1 = Validate Anim Sequence Exists!() -> (then, Path:String) {is=title K2Node_FunctionEntry}
   > n2 = Branch(Condition=ValidatePath!(InPath:String=@n1) -> (bDidSucceed:bool, OutReason:Text))
   > n3 = DoesAssetExist(AssetPath=@n1)
   > n4 = Branch(Condition=@n3)
   > SetRenderOpacity(InOpacity=1)
   > SetColorAndOpacity(self:User Widget Object Reference, InColorAndOpacity:Linear Color Structure=(R=1,G=1,B=1,A=1))
   > SetToolTipText!(execute, self:Widget Object Reference, InToolTipText:"Text (by ref)"=MakeLiteralText()) -> (then)
   > Set CanCreate(CanCreate:bool=false) -> (Output_Get:bool)
   SetRenderOpacity(execute=@n4.else, InOpacity=0.5)
   > SetColorAndOpacity(InColorAndOpacity=(R=0.3698,G=0.3469,B=0.2311,A=1))
   n5 = Split(SourceString=@n1, InStr=/, SearchDir=FromEnd)
   > SetToolTipText(InToolTipText=Format Text!(Format="LOCTEXT(\"Double-Click to Find/Create: {anim}\")", anim:String=Split(SourceString=@n5.RightS, InStr=., SearchDir=FromStart)) -> (Result))
   > Set CanCreate(CanCreate=true)
   RemoveItem!(
    execute=@n2.else
    self:List View Object Reference=Get ListView!(self:Animation Browser Object Reference=Get AnimationBrowser()) -> (ListView:List View Object Reference)
    Item:Object Reference=Get Item Data()
   ) -> (then)
  ]}
]
variables [
 CanCreate {type=boolean}
 Color And Opacity {type=SlateColor, default {colorUseRule=UseColor_Specified, specifiedColor=<rgba 1 0 1 1>}}
 Item Data {type=Object_AnimationBrowser_C, default=null}
]
widgets {
 info {rootWidgetClass=&Border, widgetCount=10}
 widgets [
  Border {widgetClassPath=&Border, bIsVariable=true, widget=Border, properties {brushColor=<rgba 1 1 1 0>}}
  HorizontalBox_1 {widgetClassPath=&HorizontalBox, parent=Border, widget=HorizontalBox_1, slot {class=BorderSlot}}
  Icon {
   widgetClassPath=&Image
   bIsVariable=true
   parent=HorizontalBox_1
   widget=Icon
   properties {brush {
     bIsDynamicallyLoaded=false
     drawAs=Image
     imageType=NoImage
     mirroring=NoMirror
     resourceName=None
     resourceObject=&Icon_LevelSequence
     tiling=NoTile
     imageSize=<14 14>
     margin=<ltrb 0 0 0 0>
     outlineSettings {
      bUseBrushTransparency=false
      roundingType=HalfHeightRadius
      width=0
      color {colorUseRule=UseColor_Specified, specifiedColor=<rgba 0 0 0 0>}
      cornerRadii=<0 0 0 0>
     }
     tintColor {colorUseRule=UseColor_Specified, specifiedColor=<rgba 1 1 1 1>}
     uVRegion {bIsValid=false, max=<0 0>, min=<0 0>}
    }}
   slot {class=HorizontalBoxSlot, properties {horizontalAlignment=HAlign_Center, verticalAlignment=VAlign_Center}}
  }
  SizeBox_1 {
   widgetClassPath=&SizeBox
   parent=HorizontalBox_1
   widget=SizeBox_1
   properties {bOverride_MinDesiredWidth=true, minDesiredWidth=200}
   slot {class=HorizontalBoxSlot, properties {padding=<ltrb 3 0 0 0>, size {sizeRule=Fill, value=1}}}
  }
  HorizontalBox_0 {
   widgetClassPath=&HorizontalBox
   parent=SizeBox_1
   widget=HorizontalBox_0
   slot {class=SizeBoxSlot, properties {verticalAlignment=VAlign_Center}}
  }
  TextBlock {
   widgetClassPath=&TextBlock
   bIsVariable=true
   parent=HorizontalBox_0
   widget=TextBlock
   properties {
    clipping=ClipToBounds
    text=AnimationSequence
    colorAndOpacity {colorUseRule=UseColor_Specified, specifiedColor=<rgba 0.5255 0.5255 0.5255 1>}
    font {
     bForceMonospaced=false
     bMaterialIsStencil=false
     fontMaterial=null
     fontObject=&Roboto
     letterSpacing=0
     monospacedWidth=1
     size=10
     skewAmount=0
     typefaceFontName=Regular
     outlineSettings {
      bApplyOutlineToDropShadows=false
      bMiteredCorners=false
      bSeparateFillAlpha=false
      outlineMaterial=null
      outlineSize=0
      outlineColor=<rgba 0 0 0 1>
     }
    }
   }
   slot {class=HorizontalBoxSlot, properties {size {sizeRule=Fill, value=1}}}
  }
  AnimMontageButton {
   widgetClassPath=&Button
   parent=HorizontalBox_0
   widget=AnimMontageButton
   properties {widgetStyle {
     clickedSlateSound {resourceObject=null}
     disabled {
      bIsDynamicallyLoaded=false
      drawAs=RoundedBox
      imageType=NoImage
      mirroring=NoMirror
      resourceName=None
      resourceObject=null
      tiling=NoTile
      imageSize=<0 0>
      margin=<ltrb 0 0 0 0>
      outlineSettings {
       bUseBrushTransparency=false
       roundingType=FixedRadius
       width=1
       color {colorUseRule=UseColor_ColorTable, specifiedColor=<rgba 0 0 0 0>}
       cornerRadii=<4 4 4 4>
      }
      tintColor {colorUseRule=UseColor_Specified, specifiedColor=<rgba 0.0395 0.0395 0.0395 1>}
      uVRegion {bIsValid=false, max=<0 0>, min=<0 0>}
     }
     disabledForeground {colorUseRule=UseColor_ColorTable, specifiedColor=<rgba 0 0 0 0>}
     hovered {
      bIsDynamicallyLoaded=false
      drawAs=NoDrawType
      imageType=NoImage
      mirroring=NoMirror
      resourceName=None
      resourceObject=null
      tiling=NoTile
      imageSize=<32 32>
      margin=<ltrb 0 0 0 0>
      outlineSettings {
       bUseBrushTransparency=false
       roundingType=FixedRadius
       width=0
       color {colorUseRule=UseColor_ColorTable, specifiedColor=<rgba 0.7243 0.7243 0.7243 0>}
       cornerRadii=<4 4 4 4>
      }
      tintColor {colorUseRule=UseColor_Specified, specifiedColor=<rgba 0.0953 0.0953 0.0953 1>}
      uVRegion {bIsValid=false, max=<0 0>, min=<0 0>}
     }
     hoveredForeground {colorUseRule=UseColor_ColorTable, specifiedColor=<rgba 0 0 0 0>}
     hoveredSlateSound {resourceObject=null}
     normal {
      bIsDynamicallyLoaded=false
      drawAs=NoDrawType
      imageType=NoImage
      mirroring=NoMirror
      resourceName=None
      resourceObject=null
      tiling=NoTile
      imageSize=<32 32>
      margin=<ltrb 0 0 0 0>
      outlineSettings {
       bUseBrushTransparency=false
       roundingType=FixedRadius
       width=1
       color {colorUseRule=UseColor_ColorTable, specifiedColor=<rgba 0 0 0 0>}
       cornerRadii=<4 4 4 4>
      }
      tintColor {colorUseRule=UseColor_Specified, specifiedColor=<rgba 0.0395 0.0395 0.0395 1>}
      uVRegion {bIsValid=false, max=<0 0>, min=<0 0>}
     }
     normalForeground {colorUseRule=UseColor_ColorTable, specifiedColor=<rgba 0 0 0 0>}
     normalPadding=<ltrb 4 0 4 0>
     pressed {
      bIsDynamicallyLoaded=false
      drawAs=NoDrawType
      imageType=NoImage
      mirroring=NoMirror
      resourceName=None
      resourceObject=null
      tiling=NoTile
      imageSize=<32 32>
      margin=<ltrb 0 0 0 0>
      outlineSettings {
       bUseBrushTransparency=false
       roundingType=FixedRadius
       width=1
       color {colorUseRule=UseColor_ColorTable, specifiedColor=<rgba 0 0 0 0>}
       cornerRadii=<4 4 4 4>
      }
      tintColor {colorUseRule=UseColor_Specified, specifiedColor=<rgba 0.0284 0.0284 0.0284 1>}
      uVRegion {bIsValid=false, max=<0 0>, min=<0 0>}
     }
     pressedForeground {colorUseRule=UseColor_ColorTable, specifiedColor=<rgba 0 0 0 0>}
     pressedPadding=<ltrb 4 0 4 0>
     pressedSlateSound {resourceObject=null}
    }}
   slot {class=HorizontalBoxSlot, properties {horizontalAlignment=HAlign_Right, verticalAlignment=VAlign_Center}}
  }
  MontageTextBlock {
   widgetClassPath=&TextBlock
   bIsVariable=true
   parent=AnimMontageButton
   widget=MontageTextBlock
   properties {
    justification=Center
    text=AM
    toolTipText=Open AnimMontage
    colorAndOpacity {colorUseRule=UseColor_Specified, specifiedColor=<rgba 0.046 0.0679 0.401 1>}
    font {
     bForceMonospaced=false
     bMaterialIsStencil=false
     fontMaterial=null
     fontObject=&Roboto
     letterSpacing=0
     monospacedWidth=1
     size=9.75
     skewAmount=0
     typefaceFontName=Bold
     outlineSettings {
      bApplyOutlineToDropShadows=false
      bMiteredCorners=false
      bSeparateFillAlpha=false
      outlineMaterial=null
      outlineSize=0
      outlineColor=<rgba 1 1 1 1>
     }
    }
   }
   slot {class=ButtonSlot, properties {padding=<ltrb 0 0 0 0>}}
  }
  SubFolderBorder {
   widgetClassPath=&EditorUtilityButton
   parent=HorizontalBox_0
   widget=SubFolderBorder
   properties {
    bIsEnabled=false
    widgetStyle {
     clickedSlateSound {resourceObject=null}
     disabled {
      bIsDynamicallyLoaded=false
      drawAs=RoundedBox
      imageType=NoImage
      mirroring=NoMirror
      resourceName=None
      resourceObject=null
      tiling=NoTile
      imageSize=<32 32>
      margin=<ltrb 0 0 0 0>
      outlineSettings {
       bUseBrushTransparency=true
       roundingType=FixedRadius
       width=1
       color {colorUseRule=UseColor_ColorTable, specifiedColor=<rgba 0 0 0 0>}
       cornerRadii=<10 10 10 10>
      }
      tintColor {colorUseRule=UseColor_Specified, specifiedColor=<rgba 0 0.0369 0.0395 1>}
      uVRegion {bIsValid=false, max=<0 0>, min=<0 0>}
     }
     disabledForeground {colorUseRule=UseColor_ColorTable, specifiedColor=<rgba 0 0 0 0>}
     hovered {
      bIsDynamicallyLoaded=false
      drawAs=RoundedBox
      imageType=NoImage
      mirroring=NoMirror
      resourceName=None
      resourceObject=null
      tiling=NoTile
      imageSize=<32 32>
      margin=<ltrb 0 0 0 0>
      outlineSettings {
       bUseBrushTransparency=false
       roundingType=FixedRadius
       width=1
       color {colorUseRule=UseColor_ColorTable, specifiedColor=<rgba 0 0 0 0>}
       cornerRadii=<4 4 4 4>
      }
      tintColor {colorUseRule=UseColor_Specified, specifiedColor=<rgba 0.0953 0.0953 0.0953 1>}
      uVRegion {bIsValid=false, max=<0 0>, min=<0 0>}
     }
     hoveredForeground {colorUseRule=UseColor_ColorTable, specifiedColor=<rgba 0 0 0 0>}
     hoveredSlateSound {resourceObject=null}
     normal {
      bIsDynamicallyLoaded=false
      drawAs=RoundedBox
      imageType=NoImage
      mirroring=NoMirror
      resourceName=None
      resourceObject=null
      tiling=NoTile
      imageSize=<32 32>
      margin=<ltrb 0 0 0 0>
      outlineSettings {
       bUseBrushTransparency=true
       roundingType=FixedRadius
       width=1
       color {colorUseRule=UseColor_ColorTable, specifiedColor=<rgba 0 0 0 0>}
       cornerRadii=<10 10 10 10>
      }
      tintColor {colorUseRule=UseColor_Specified, specifiedColor=<rgba 0 0.0369 0.0395 1>}
      uVRegion {bIsValid=false, max=<0 0>, min=<0 0>}
     }
     normalForeground {colorUseRule=UseColor_ColorTable, specifiedColor=<rgba 0 0 0 0>}
     normalPadding=<ltrb 12 1.5 12 1.5>
     pressed {
      bIsDynamicallyLoaded=false
      drawAs=RoundedBox
      imageType=NoImage
      mirroring=NoMirror
      resourceName=None
      resourceObject=null
      tiling=NoTile
      imageSize=<32 32>
      margin=<ltrb 0 0 0 0>
      outlineSettings {
       bUseBrushTransparency=false
       roundingType=FixedRadius
       width=1
       color {colorUseRule=UseColor_ColorTable, specifiedColor=<rgba 0 0 0 0>}
       cornerRadii=<4 4 4 4>
      }
      tintColor {colorUseRule=UseColor_Specified, specifiedColor=<rgba 0.0284 0.0284 0.0284 1>}
      uVRegion {bIsValid=false, max=<0 0>, min=<0 0>}
     }
     pressedForeground {colorUseRule=UseColor_ColorTable, specifiedColor=<rgba 0 0 0 0>}
     pressedPadding=<ltrb 12 2.5 12 0.5>
     pressedSlateSound {resourceObject=null}
    }
   }
   slot {class=HorizontalBoxSlot}
  }
  SubFolderTextBlock {
   widgetClassPath=&TextBlock
   bIsVariable=true
   parent=SubFolderBorder
   widget=SubFolderTextBlock
   properties {
    text=Extra Text
    colorAndOpacity {colorUseRule=UseColor_Specified, specifiedColor=<rgba 0.5271 0.5271 0.5271 1>}
    font {
     bForceMonospaced=false
     bMaterialIsStencil=false
     fontMaterial=null
     fontObject=&Roboto
     letterSpacing=0
     monospacedWidth=1
     size=10
     skewAmount=0
     typefaceFontName=Regular
     outlineSettings {
      bApplyOutlineToDropShadows=false
      bMiteredCorners=false
      bSeparateFillAlpha=false
      outlineMaterial=null
      outlineSize=0
      outlineColor=<rgba 0 0 0 1>
     }
    }
   }
   slot {class=ButtonSlot, properties {padding=<ltrb 0 0 0 0>}}
  }
 ]
}
