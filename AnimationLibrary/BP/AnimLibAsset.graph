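# Editor-side wrapper for one Animation Library asset, mirroring a Python asset handle's metadata, source sequence, Control Rig and thumbnail as Blueprint variables.
# Embedded Python handles edit, save and delete through the library manager, with edits re-running initialize to refresh the variables.
# Dirty and writeable delegates broadcast state, and PollDirty completes the save notification.
# Thumbnail updates take high-res screenshots, then poll for the last frame file with limited retries before broadcasting.
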EditorUtilityBlueprint /AnimationLibrary/BP/AnimLibAsset
graphs [
 EventGraph {nodes [
   n1 = OnThumbUpdate!() -> (OutputDelegate, then, ThumbPath:String, Task:Automation Editor Task Object Reference, WaitOnFile:String) {is=title K2Node_CustomEvent}
   > Branch(Condition=IsValidTask!(self:Automation Editor Task Object Reference=@n1.Task) -> (ReturnValue:bool))
   > Set ThumbRetryCount(ThumbRetryCount:int=60) -> (Output_Get:int=0)
   n2 = Get UnrealEditorSubsystem()
   > n3 = GetEditorWorld(self=@n2)
   Delay!(execute=[@n3.then, @n5.then], WorldContextObject:Object Reference=@n3, Duration=0.5) -> (then)
   > n4 = Branch(Condition=FileExists!(InPath:String=@n1.WaitOnFile) -> (ReturnValue:bool))
   SetSequencePath!(
    execute=[@n6.then, @n4.then]
    self:Img Media Source Object Reference=Get ThumbnailMediaSource() -> (ThumbnailMediaSource:Img Media Source Object Reference)
    Path:String=@n1.ThumbPath
   ) -> (then)
   > set_dirty!(execute, self, new_dirty:bool=true) -> (then, NewParam:bool)
   > Call Thumbnail Updated!(execute:Exec, self:Anim Lib Asset Object Reference, ThumbPath:String=@n1.ThumbPath, Task:Automation Editor Task Object Reference=@n1.Task) -> (then:Exec) {is=title K2Node_CallDelegate}
   Decrement Int(" "=@n4.else, Value=Get ThumbRetryCount() -> (ThumbRetryCount:int=0))
   > n5 = Branch(Condition="integer > integer"(A=Get ThumbRetryCount() -> (ThumbRetryCount=0)))
   n6 = PrintString(execute=@n5.else, InString=Append(A="Thumb capture timed out: ", B=@n1.ThumbPath))
   InEditorOrPIE!(WorldContextObject:Object Reference=@n2) -> (ReturnValue:bool)
  ]}
 OnDirtyChanged {nodes [OnDirtyChanged!() -> (then, isDirty:bool=~) {is=title K2Node_FunctionEntry}]}
 OnWriteableChanged {nodes [OnWriteableChanged!() -> (then, IsWriteable:bool=~) {is=title K2Node_FunctionEntry}]}
 PollDirty {nodes [
   PollDirty() {is=title K2Node_FunctionEntry}
   > n1 = is_valid!(execute, self) -> (then, valid:bool)
   > Branch(Condition=@n1)
   > n2 = Execute Python Script!(
    execute
    PythonScript="import unreal\r\n\r\nasset = unreal.resolve_python_object_handle(asset_handle)\r\ndirty = asset.is_dirty"
    asset_handle:Python Object Handle Object Reference=Get AssetHandle() -> (AssetHandle:Python Object Handle Object Reference)
   ) -> (then, ReturnValue, dirty:bool=~) {inputs [asset_handle], outputs [dirty]}
   > Branch(Condition="Not Equal (Boolean)"(A=Get dirty() -> (dirty:bool), B=@n2.dirty) -> (ReturnValue=~))
   > n3 = Set dirty(dirty:bool=@n2.dirty) -> (Output_Get:bool)
   > Call On Dirty Changed!(execute:Exec, self:Anim Lib Asset Object Reference, isDirty:bool=@n3) -> (then:Exec) {is=title K2Node_CallDelegate}
   > n4 = Branch(Condition=@n3)
   Return Node(execute=[@n7.then, @n4.then])
   n5 = Get SaveNotification() -> (SaveNotification:Editor Notification Handle Structure)
   SetCompletionState!(
    execute=@n4.else
    NotificationHandle:"Editor Notification Handle Structure (by ref)"=@n5
    State:EEditorNotificationCompletionState Enum=Success
   ) -> (then)
   n6 = Get Name() -> (Name:String)
   > SetSubText!(
    execute
    NotificationHandle:"Editor Notification Handle Structure (by ref)"=@n5
    SubText:Text="To Text (String)"(InString=Append(A=@n6, B=" SAVED!"))
   ) -> (then, ReturnValue:Editor Notification Handle Structure)
   > n7 = ExpireAndFadeOut!(execute, NotificationHandle:"Editor Notification Handle Structure (by ref)"=@n5) -> (then)
  ]}
 SelectControlsInrig {nodes [
   SelectControlsInrig() {is=title K2Node_FunctionEntry}
   > Set tempRig(tempRig:Control Rig Object Reference) -> (Output_Get:Control Rig Object Reference)
   > n1 = GetFocusedLevelSequence!(execute) -> (then, ReturnValue:Level Sequence Object Reference)
   > n2 = GetControlRigs!(execute, LevelSequence:Level Sequence Object Reference=@n1) -> (then, ReturnValue:Array of Control Rig Sequencer Binding Proxy Structures)
   > n3 = For Each Loop with Break!(Exec:Exec, Array:Array of Control Rig Sequencer Binding Proxy Structures=@n2, Break:Exec=@n8.then) -> (LoopBody:Exec, Array Element:Control Rig Sequencer Binding Proxy Structure, Array Index:int, Completed:Exec)
   n4 = Get tempRig() -> (tempRig:Control Rig Object Reference)
   ClearControlSelection!(execute=@n3.Completed, self:Control Rig Object Reference=@n4, bSetupUndo:bool=true) -> (then, ReturnValue:bool)
   > n5 = For Each Loop(Array=Get Controls() -> (Controls:Array of Strings))
   Return Node!(execute=@n5.Completed, Control Rig:Control Rig Object Reference)
   > SelectControl!(
    execute
    self:Control Rig Object Reference=@n4
    InControlName:"Name (by ref)"=String To Name(InString=@n5)
    bSelect:bool=true
    bSetupUndo:bool=true
   ) -> (then)
   n6 = Break Control Rig Sequencer Binding Proxy!(ControlRigSequencerBindingProxy:"Control Rig Sequencer Binding Proxy Structure (by ref)"=@n3) -> (
    Proxy:Movie Scene Binding Proxy Structure
    ControlRig:Control Rig Object Reference
    Track:Movie Scene Control Rig Parameter Track Object Reference
   ) {is=title K2Node_BreakStruct}
   n7 = Execute Python Script!(
    execute=@n3.LoopBody
    PythonScript="import unreal\r\npackage = obj.get_class().get_package()\r\nrig_bp = unreal.load_asset(package.get_path_name())"
    obj:Control Rig Object Reference=@n6.ControlRig
   ) -> (then, ReturnValue, rig_bp:Object Reference) {inputs [obj], outputs [rig_bp]}
   > Branch(Condition="Equal (Object)"(A=@n7.rig_bp, B=Get Rig() -> (Rig:Control Rig Blueprint Object Reference)))
   > n8 = Set tempRig(tempRig=@n6.ControlRig)
  ]}
 StartSavingNotification {nodes [
   StartSavingNotification() {is=title K2Node_FunctionEntry}
   n1 = Append(A=SAVING:, B=Get Name())
   > n2 = AddEditorNotification!(
    execute
    EditorNotificationSettings:"Editor Notification Settings Structure (by ref)"=MakeEditorNotificationSettings!(
     Text:Text="LOCTEXT(\"Animation Library\")"
     SubText:Text="To Text (String)"(InString=@n1)
     FadeInDuration:float=0.5
     FadeOutDuration:float=2
     ExpireDuration:float=1
     bUseThrobber:bool=true
     bUseSuccessFailIcons:bool=true
     bFireAndForget:bool=false
     bAllowThrottleWhenFrameRateIsLow:bool=true
    ) -> (ReturnValue:Editor Notification Settings Structure)
   ) -> (then, ReturnValue:Editor Notification Handle Structure)
   > n3 = Set SaveNotification(SaveNotification:Editor Notification Handle Structure=@n2) -> (Output_Get:Editor Notification Handle Structure)
   > SetCompletionState(NotificationHandle=@n3, State=Pending)
  ]}
 ThumbnailUpdated {nodes [ThumbnailUpdated!() -> (then, ThumbPath:String, Task:Automation Editor Task Object Reference) {is=title K2Node_FunctionEntry}]}
 delete {nodes [
   delete() {is=title K2Node_FunctionEntry}
   > Execute Python Script!(
    execute
    PythonScript="import unreal\r\n\r\nasset = unreal.resolve_python_object_handle(asset_handle)\r\nmanager  = unreal.resolve_python_object_handle(in_manager)\r\nmanager.remove(asset)\r\n"
    asset_handle:Python Object Handle Object Reference=Get AssetHandle()
    in_manager:Python Object Handle Object Reference=Get Manager!(self:Anim Lib Manager Object Reference=Get AssetLibrary() -> (AssetLibrary:Anim Lib Manager Object Reference)) -> (Manager:Python Object Handle Object Reference)
   ) -> (then, ReturnValue) {inputs [asset_handle,in_manager]}
  ]}
 edit_author {nodes [
   n1 = edit_author!() -> (then, NewName:String) {is=title K2Node_FunctionEntry}
   > n2 = is_valid()
   > n3 = Branch(Condition=OR Boolean(A=NOT Boolean(A=@n2), B=IsEmpty(InString=@n1)))
   n4 = GetManager!(execute=@n3.else, self:Anim Lib Manager Object Reference=Get AssetLibrary()) -> (then, GetManager:Python Object Handle Object Reference)
   > Execute Python Script!(
    execute
    PythonScript="import unreal\r\n\r\nasset = unreal.resolve_python_object_handle(asset_handle)\r\n\r\nmanager = unreal.resolve_python_object_handle(manager)\r\nwith manager.batch('Change Asset Author'):\r\n    asset.metadata.author = new_name\r\ndirty = asset.is_dirty"
    asset_handle=Get AssetHandle()
    new_name:String=@n1
    manager:Python Object Handle Object Reference=@n4
   ) -> (then, ReturnValue) {inputs [asset_handle,new_name,manager]}
   > initialize!(execute, self, handle:Python Object Handle Object Reference=Get AssetHandle()) -> (then)
  ]}
 edit_name {nodes [
   n1 = edit_name!() -> (then, NewName:String) {is=title K2Node_FunctionEntry}
   > n2 = is_valid()
   > n3 = Branch(Condition=OR Boolean(A=NOT Boolean(A=@n2), B=IsEmpty(InString=@n1)))
   n4 = GetManager(execute=@n3.else, self=Get AssetLibrary())
   > Execute Python Script(
    PythonScript="import unreal\r\n\r\nasset = unreal.resolve_python_object_handle(asset_handle)\r\nmanager = unreal.resolve_python_object_handle(manager)\r\nwith manager.batch('Rename Asset'):\r\n    asset.name = new_name\r\n"
    asset_handle=Get AssetHandle()
    new_name=@n1
    manager=@n4
   ) {inputs [asset_handle,new_name,manager]}
   > initialize(handle=Get AssetHandle())
  ]}
 edit_notes {nodes [
   n1 = edit_notes!() -> (then, new_notes:String) {is=title K2Node_FunctionEntry}
   > n2 = is_valid()
   > Branch(Condition=@n2)
   > n3 = GetManager(self=Get AssetLibrary())
   > Execute Python Script!(
    execute
    PythonScript="import unreal\r\n\r\nasset = unreal.resolve_python_object_handle(asset_handle)\r\nmanager = unreal.resolve_python_object_handle(manager)\r\nwith manager.batch('Edit Asset Notes'):\r\n    asset.metadata.note = new_notes\r\ndirty = asset.is_dirty"
    asset_handle=Get AssetHandle()
    new_notes:String=@n1
    manager=@n3
   ) -> (then, ReturnValue) {inputs [asset_handle,new_notes,manager]}
   > initialize(handle=Get AssetHandle())
  ]}
 edit_path {nodes [
   n1 = edit_path!() -> (then, new_path:String) {is=title K2Node_FunctionEntry}
   > n2 = is_valid()
   > Branch(Condition=@n2)
   > n3 = GetManager(self=Get AssetLibrary())
   > n4 = Execute Python Script!(
    execute
    PythonScript="import unreal\r\n\r\nmanager = unreal.resolve_python_object_handle(in_manager)\r\nasset = unreal.resolve_python_object_handle(asset_handle)\r\n\r\nis_valid = manager.is_valid_path(new_path=new_path)\r\nif is_valid:\r\n    with manager.batch('Move Asset'):       \r\n        asset.virtual_path = new_path\r\n\r\nnew_path_out = asset.virtual_path.as_posix()\r\ndirty = asset.is_dirty"
    asset_handle=Get AssetHandle()
    new_path:String=@n1
    in_manager:Python Object Handle Object Reference=@n3
   ) -> (then, ReturnValue, new_path_out:String, is_valid:bool=~) {inputs [asset_handle,new_path,in_manager], outputs [new_path_out,is_valid]}
   > initialize(handle=Get AssetHandle())
   > Branch(Condition=@n4)
   > Return Node!(execute, validated_path:String=@n4.new_path_out, is_valid_path:bool=@n4.is_valid)
  ]}
 edit_tags {nodes [
   n1 = edit_tags!() -> (then, NewTags:String) {is=title K2Node_FunctionEntry}
   > n2 = is_valid()
   > Branch(Condition=@n2)
   > n3 = GetManager(self=Get AssetLibrary())
   > Execute Python Script!(
    execute
    PythonScript="import unreal\r\n\r\nasset = unreal.resolve_python_object_handle(asset_handle)\r\nmanager = unreal.resolve_python_object_handle(in_manager)\r\n\r\nwith manager.batch('Change Asset Tags'):\r\n    asset.metadata.tags.clear()\r\n    manager.add_tags(asset, new_tags)\r\ndirty = asset.is_dirty"
    asset_handle=Get AssetHandle()
    new_tags:String=@n1
    in_manager=@n3
   ) -> (then, ReturnValue) {inputs [asset_handle,new_tags,in_manager]}
   > initialize(handle=Get AssetHandle())
  ]}
 initialize {nodes [
   n1 = initialize!() -> (then, handle:Python Object Handle Object Reference) {is=title K2Node_FunctionEntry}
   > n2 = Set AssetHandle(AssetHandle:Python Object Handle Object Reference=@n1) -> (Output_Get:Python Object Handle Object Reference)
   > n3 = is_valid()
   > n4 = Branch(Condition=@n3)
   > n5 = Execute Python Script!(
    execute
    PythonScript="import unreal\r\n\r\nasset = unreal.resolve_python_object_handle(asset_handle)\r\nmanager = unreal.resolve_python_object_handle(in_manager)\r\nid = str(asset.id)\r\nname = asset.name\r\nauthor = asset.metadata.author\r\ntags = asset.metadata.tags\r\n\r\ntags = ', '.join(tags)\r\nlength = asset.length\r\npath = asset.virtual_path.as_posix()\r\ncontrols = []\r\nsections = []\r\nfor s in asset.metadata.rig_info.get('sections', []):\r\n    controls.extend(s.get('controls', []))\r\n    sections.append(s.get('name', ''))\r\nrig_name = asset.metadata.rig_info.get('rig', '')\r\nnotes = asset.metadata.note\r\ncontrols = sorted(set(controls))\r\nsequence = asset.metadata.from_asset\r\n\r\nobj_type = asset.metadata.asset_type.name.title()\r\ntooltip = str(asset)\r\nthumbnail_path = asset.metadata.thumbnail.sequence\r\ns = asset.metadata.rig_info.get('s', 0)\r\ne = asset.metadata.rig_info.get('e', 1)\r\ndirty = asset.is_dirty\r\nwriteable = manager.can_write(asset)"
    asset_handle=@n2
    in_manager=Get Manager(self=Get AssetLibrary())
   ) -> (
    then
    ReturnValue
    name:String
    author:String
    tags:String
    path:String
    controls:Array of Strings
    notes:String
    sections:Array of Strings
    rig_name:String
    sequence:String
    obj_type:String
    tooltip:String
    thumbnail_path:String
    length:int
    id:String
    s:int
    e:int
    dirty:bool=~
    writeable:bool=~
   ) {
    inputs [asset_handle,in_manager]
    outputs [
     name
     author
     tags
     path
     controls
     notes
     sections
     rig_name
     sequence
     obj_type
     tooltip
     thumbnail_path
     length
     id
     s
     e
     dirty
     writeable
    ]
   }
   > n6 = Sequence!(execute) -> (
    then_0
    then_1
    then_2:Exec
    then_3:Exec
    then_4:Exec
    then_5:Exec
    then_6:Exec
    then_7:Exec
    then_8:Exec
    then_9:Exec
    then_10:Exec
    then_11:Exec
    then_12:Exec
    then_13:Exec
    then_14:Exec
    then_15:Exec
    then_16:Exec
    then_17:Exec
   )
   n7 = GetAssetRegistry()
   > n8 = Get Asset By Object Path!(
    execute
    self:Asset Registry Interface=@n7
    ObjectPath:"Soft Object Path Structure (by ref)"=MakeSoftObjectPath!(PathString:String=@n5.rig_name) -> (ReturnValue:Soft Object Path Structure)
    bIncludeOnlyOnDiskAssets:bool=false
    bSkipARFilteredAssets:bool=true
   ) -> (then, ReturnValue:Asset Data Structure) {is=title K2Node_CallFunction}
   > Branch(Condition=Is Valid Asset Data!(InAssetData:"Asset Data Structure (by ref)"=@n8) -> (ReturnValue:bool))
   > n9 = CastToControlRigBlueprint!(execute, Object:Object Reference=GetAsset(InAssetData=@n8)) -> (CastSucceeded:Exec, CastFailed:Exec, AsControlRigBlueprint:Control Rig Blueprint Object Reference)
   > Set Rig(Rig:Control Rig Blueprint Object Reference=@n9) -> (Output_Get:Control Rig Blueprint Object Reference)
   n10 = Get Asset By Object Path(
    execute=@n6.then_1
    self=@n7
    ObjectPath=MakeSoftObjectPath(PathString=@n5.sequence)
    bIncludeOnlyOnDiskAssets=false
    bSkipARFilteredAssets=true
   )
   > Branch(Condition=Is Valid Asset Data(InAssetData=@n10))
   > n11 = Cast To LevelSequence!(execute, Object=GetAsset(InAssetData=@n10)) -> (then, CastFailed, AsLevel Sequence:Level Sequence Object Reference) {is=title K2Node_DynamicCast}
   > Set FromSequence(FromSequence:Level Sequence Object Reference=@n11) -> (Output_Get:Level Sequence Object Reference)
   Set ToolTip(execute=@n6.then_10, ToolTip:String=@n5.tooltip) -> (Output_Get:String)
   Set Length(execute=@n6.then_11, Length:int=@n5.length) -> (Output_Get:int=0)
   n12 = Construct Object from Class(execute=@n6.then_12, Class=/Script/ImgMedia.ImgMediaSource) -> (ReturnValue:Img Media Source Object Reference)
   > n13 = Set ThumbnailMediaSource(ThumbnailMediaSource:Img Media Source Object Reference=@n12) -> (Output_Get:Img Media Source Object Reference)
   > SetSequencePath(self=@n13, Path=@n5.thumbnail_path)
   Set id(execute=@n6.then_13, id:String=@n5.id) -> (Output_Get:String)
   Set Start(execute=@n6.then_14, Start:int=@n5.s) -> (Output_Get:int=0)
   Set End(execute=@n6.then_15, End:int=@n5.e) -> (Output_Get:int=0)
   n14 = Set dirty(execute=@n6.then_16, dirty=@n5.dirty)
   > Call On Dirty Changed(isDirty=@n14)
   n15 = Set is_writeable(execute=@n6.then_17, is_writeable:bool=@n5.writeable) -> (Output_Get:bool)
   > Call On Writeable Changed!(execute:Exec, self:Anim Lib Asset Object Reference, IsWriteable:bool=@n15) -> (then:Exec) {is=title K2Node_CallDelegate}
   Set Name(execute=@n6.then_2, Name:String=@n5.name) -> (Output_Get:String)
   Set Author(execute=@n6.then_3, Author:String=@n5.author) -> (Output_Get:String)
   Set Tags(execute=@n6.then_4, Tags:String=@n5.tags) -> (Output_Get:String)
   Set Path(execute=@n6.then_5, Path:String=@n5.path) -> (Output_Get:String)
   Set Notes(execute=@n6.then_6, Notes:String=@n5.notes) -> (Output_Get:String)
   Set Controls(execute=@n6.then_7, Controls:Array of Strings=@n5.controls) -> (Output_Get:Array of Strings)
   Set Sections(execute=@n6.then_8, Sections:Array of Strings=@n5.sections) -> (Output_Get:Array of Strings)
   Set Type(execute=@n6.then_9, Type:String=@n5.obj_type) -> (Output_Get:String)
   n16 = Sequence!(execute=@n4.else) -> (then_0, then_1, then_2, then_3, then_4, then_5, then_6, then_7, then_8, then_9, then_10, then_11, then_12)
   > Set Name()
   Set Author(execute=@n16.then_1)
   Set FromSequence(execute=@n16.then_10)
   Set ThumbnailMediaSource(execute=@n16.then_11)
   Set Length(execute=@n16.then_12, Length=0) -> (Output_Get=0)
   Set Tags(execute=@n16.then_2)
   Set Path(execute=@n16.then_3)
   Set Notes(execute=@n16.then_4)
   Clear(execute=@n16.then_5, TargetArray=Get Controls())
   Clear(execute=@n16.then_6, TargetArray=Get Sections() -> (Sections:Array of Strings))
   Set Type(execute=@n16.then_7)
   Set ToolTip(execute=@n16.then_8)
   Set Rig(execute=@n16.then_9)
  ]}
 is_valid {nodes [
   is_valid() {is=title K2Node_FunctionEntry}
   > Return Node!(execute, valid:bool=IsValid(Object=Get AssetHandle()))
  ]}
 save_asset {nodes [
   save_asset() {is=title K2Node_FunctionEntry}
   > n1 = is_valid() {is=title K2Node_CallFunction}
   > Branch(Condition=@n1)
   > Execute Python Script!(
    execute
    PythonScript="import unreal\r\n\r\nasset = unreal.resolve_python_object_handle(asset_handle)\r\nmanager = unreal.resolve_python_object_handle(in_manager)\r\nmanager.save(asset)"
    asset_handle=Get AssetHandle()
    in_manager=Get Manager(self=Get AssetLibrary())
   ) -> (then, ReturnValue) {inputs [asset_handle,in_manager]}
   > StartSavingNotification() {is=title K2Node_CallFunction}
  ]}
 set_dirty {nodes [
   n1 = set_dirty!() -> (then, new_dirty:bool=~) {is=title K2Node_FunctionEntry}
   > n2 = is_valid()
   > Branch(Condition=@n2)
   > n3 = Execute Python Script!(
    execute
    PythonScript="import unreal\r\n\r\nasset = unreal.resolve_python_object_handle(asset_handle)\r\nif new_dirty:\r\n    asset.mark_dirty()\r\nelse:\r\n    asset.mark_saved()\r\ndirty = asset.is_dirty"
    asset_handle=Get AssetHandle()
    new_dirty:bool=@n1
   ) -> (then, ReturnValue, dirty:bool=~) {inputs [asset_handle,new_dirty], outputs [dirty]}
   > PollDirty() {is=title K2Node_CallFunction}
   > Return Node!(execute, NewParam:bool=@n3.dirty)
  ]}
 switch_source {nodes [switch_source() {is=title K2Node_FunctionEntry}]}
 update_thumbnail {nodes [
   update_thumbnail() {is=title K2Node_FunctionEntry}
   > n1 = Show Message Dialog(
    Title=MakeLiteralText(Value="LOCTEXT(\"Animation Library\")")
    Message=MakeLiteralText(Value="LOCTEXT(\"You\\'re about to generate this asset\\'s thumbnail:\\r\\nYes: Continue.\\r\\nNo: Skip it.\")")
    MessageType=YesNo
    DefaultValue=No
    MessageCategory=Warning
   )
   > n2 = Switch on EAppReturnType(Selection=@n1)
   n3 = Get AssetHandle()
   n4 = Execute Python Script!(
    execute=[@n6.then, @n2.No]
    PythonScript="import unreal\r\n\r\nasset = unreal.resolve_python_object_handle(asset_handle)\r\nthmb = asset.metadata.thumbnail\r\nthmb.prepare_capture()\r\nsequence = thmb.sequence\r\nwait_on_file = thmb.last_frame\r\n"
    asset_handle=@n3
   ) -> (then, ReturnValue, sequence:String, wait_on_file:String) {inputs [asset_handle], outputs [sequence,wait_on_file]}
   > n5 = TakeHighResScreenshot!(
    execute
    ResX:int=256
    ResY:int=256
    Filename:String=@n4.sequence
    Camera:Camera Actor Object Reference
    bMaskEnabled:bool=false
    bCaptureHDR:bool=false
    ComparisonTolerance:EComparisonTolerance Enum=Low
    ComparisonNotes:String
    Delay:float=0
    bForceGameView:bool=true
   ) -> (then, ReturnValue:Automation Editor Task Object Reference)
   > OnThumbUpdate!(
    execute
    self=Self-Reference()
    ThumbPath:String=@n4.sequence
    Task:Automation Editor Task Object Reference=@n5
    WaitOnFile:String=@n4.wait_on_file
   ) -> (then) {is=title K2Node_CallFunction}
   > AddSimpleEditorNotification!(
    execute
    Text:Text="LOCTEXT(\"Animation Library\")"
    SubText:Text="LOCTEXT(\"Thumbnail Updated\")"
    FadeInDuration:float=0.5
    FadeOutDuration:float=2
    ExpireDuration:float=1
   ) -> (then)
   > Return Node!(execute, SequencePath:String=@n4.sequence)
   n6 = Execute Python Script!(
    execute=@n2.Yes
    PythonScript="import unreal\r\n\r\nasset = unreal.resolve_python_object_handle(asset_handle)\r\nthmb = asset.metadata.thumbnail\r\nthmb.capture()\r\n\r\n"
    asset_handle=@n3
   ) -> (then, ReturnValue) {inputs [asset_handle]}
  ]}
]
variables [
 AssetHandle {type=PythonObjectHandle, default=null}
 AssetLibrary {type=AnimLibManager_C, default=null}
 Author {type=string, default=""}
 Controls {type="array<string>", default []}
 dirty {type=boolean}
 End {type=integer}
 FromSequence {type=LevelSequence, default=null}
 id {type=string, default=""}
 is_writeable {type=boolean}
 Length {type=integer}
 Name {type=string, default=""}
 Notes {type=string, default=""}
 OnDirtyChanged {default=()}
 OnWriteableChanged {default=()}
 Path {type=string, default=""}
 Rig {type=ControlRigBlueprint, default=null, class=~}
 SavedNotification__ {
  type=EditorNotificationSettings
  default {
   bAllowThrottleWhenFrameRateIsLow=true
   bFireAndForget=true
   bUseSuccessFailIcons=true
   bUseThrobber=false
   expireDuration=1
   fadeInDuration=0
   fadeOutDuration=0
   subText=""
   text=""
  }
 }
 SaveNotification {type=EditorNotificationHandle, default {}}
 Sections {type="array<string>", default []}
 Start {type=integer}
 Tags {type=string, default=""}
 thumbnail_source {type=string, default=""}
 ThumbnailMediaSource {type=ImgMediaSource, default=null}
 ThumbnailUpdated {default=()}
 ThumbRetryCount {type=integer, default=60}
 ToolTip {type=string, default=""}
 Type {type=string, default=""}
]
