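# Editor utility Blueprint that represents a named selection set of Control Rig controls in the animation library.
# It stores a name and a list of control names.
# Select optionally clears the given rig's selection, then selects each stored control without setting up undo.
# Serialize and deserialize use embedded Python to convert both variables to and from a dictionary passed as a Python object handle.
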EditorUtilityBlueprint /AnimationLibrary/BP/ControlsSelectionSet
graphs [
 EventGraph {links [], nodes []}
 Select {nodes [
   n1 = Select!() -> (then, Rig:"Control Rig Object Reference (by ref)", Clear:bool=true) {is=title K2Node_FunctionEntry}
   > n2 = Branch(Condition=@n1.Clear)
   > n3 = ClearControlSelection!(execute, self:Control Rig Object Reference=@n1, bSetupUndo:bool=false) -> (then, ReturnValue:bool)
   n4 = For Each Loop(Exec=[@n3.then, @n2.else], Array:Array of Names=Get Controls() -> (Controls:Array of Names)) -> (Array Element:Name)
   > SelectControl!(execute, self:Control Rig Object Reference=@n1, InControlName:"Name (by ref)"=@n4, bSelect:bool=true, bSetupUndo:bool=false) -> (then)
  ]}
 deserialize {nodes [
   n1 = deserialize!() -> (then, dict:Python Object Handle Object Reference) {is=title K2Node_FunctionEntry}
   > n2 = Execute Python Script!(
    execute
    PythonScript="import unreal\r\n\r\ndata = unreal.resolve_python_object_handle(input)\r\nname = data.get('name', '')\r\ncontrols = [unreal.Name(c) for c in data.get('controls', [])]"
    input:Python Object Handle Object Reference=@n1
   ) -> (then, ReturnValue, name:String, controls:Array of Names) {inputs [input], outputs [name,controls]}
   > Set Name(Name:String=@n2.name) -> (Output_Get:String)
   > Set Controls(Controls:Array of Names=@n2.controls) -> (Output_Get:Array of Names)
  ]}
 serialize {nodes [
   serialize() {is=title K2Node_FunctionEntry}
   > n1 = Execute Python Script!(
    execute
    PythonScript="import unreal\r\n\r\noutput = unreal.create_python_object_handle({\r\n            'name': name,\r\n            'controls': [str(c) for c in controls]\r\n        })"
    name:String=Get Name() -> (Name:String)
    controls:Array of Names=Get Controls()
   ) -> (then, ReturnValue, output:Python Object Handle Object Reference) {inputs [name,controls], outputs [output]}
   > Return Node!(execute, dict:Python Object Handle Object Reference=@n1.output)
  ]}
]
variables [
 Controls {type="array<string>", default []}
 Name {type=string, default=""}
]
