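# Editor utility Blueprint that manages an animation library, bridging a Python Manager to Blueprint asset and selection-set objects.
# Bootstrap purges cached AnimLibrary modules, wraps assets, settings and sources as Python handles, then starts a looping timer.
# Refresh polling either rebuilds every asset or reconciles created, updated and removed IDs, then broadcasts OnRefreshAssets.
# Filtering rebuilds assets from Python searches; shutdown prompts to save dirty assets and clears the timer.
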EditorUtilityBlueprint /AnimationLibrary/BP/AnimLibManager
graphs [
 AddAsset {nodes [
   n1 = AddAsset!() -> (then, handle:Python Object Handle Object Reference) {is=title K2Node_FunctionEntry}
   > n2 = Sequence()
   > n3 = Construct Object from Class(Class=&/AnimLibAsset.AnimLibAsset_C) -> (ReturnValue:Anim Lib Asset Object Reference)
   > Set AssetLibrary!(execute, AssetLibrary:Anim Lib Manager Object Reference=Self-Reference(), self:Anim Lib Asset Object Reference=@n3) -> (then, Output_Get:Anim Lib Manager Object Reference)
   > initialize!(execute, self:Anim Lib Asset Object Reference=@n3, handle:Python Object Handle Object Reference=@n1) -> (then)
   n4 = Execute Python Script!(
    execute=@n2.then_1
    PythonScript="import unreal\r\n\r\nasset = unreal.resolve_python_object_handle(handle)\r\nid_str = str(asset.id)"
    handle:Python Object Handle Object Reference=@n1
   ) -> (then, ReturnValue, id_str:"String (by ref)") {inputs [handle], outputs [id_str]}
   > Add!(
    execute
    TargetMap:Map of Strings to Anim Lib Asset Object References=Get LibAssets() -> (LibAssets:Map of Strings to Anim Lib Asset Object References)
    Key:"String (by ref)"=@n4.id_str
    Value:"Anim Lib Asset Object Reference (by ref)"=@n3
   ) -> (then) {is=title K2Node_CallFunction}
   > Return Node!(execute, Asset:Anim Lib Asset Object Reference=@n3)
  ]}
 AddSelectionSet {nodes [
   n1 = AddSelectionSet!() -> (then, Name:String, Controls:Array of Names) {is=title K2Node_FunctionEntry}
   > n2 = Construct Object from Class(Class=&/ControlsSelectionSet.ControlsSelectionSet_C) -> (ReturnValue:Controls Selection Set Object Reference)
   > Set Name!(execute, Name:String=@n1, self:Controls Selection Set Object Reference=@n2) -> (then, Output_Get:String)
   > Set Controls!(execute, Controls:Array of Names=@n1.Controls, self:Controls Selection Set Object Reference=@n2) -> (then, Output_Get:Array of Names)
   > Add(
    TargetMap:Map of Strings to Controls Selection Set Object References=Get SelectionSets() -> (SelectionSets:Map of Strings to Controls Selection Set Object References)
    Key=@n1
    Value:"Controls Selection Set Object Reference (by ref)"=@n2
   )
   > n3 = serialize!(execute, self:Controls Selection Set Object Reference=@n2) -> (then, dict:Python Object Handle Object Reference)
   > Execute Python Script!(
    execute
    PythonScript="import unreal\r\n\r\nmanager = unreal.resolve_python_object_handle(in_handle)\r\nmanager.selection_set_manager.create_selection_set_from_handle(in_set)"
    in_handle:Python Object Handle Object Reference=Get Manager() -> (Manager:Python Object Handle Object Reference)
    in_set:Python Object Handle Object Reference=@n3
   ) -> (then, ReturnValue) {inputs [in_handle,in_set]}
   > Return Node!(execute, new_set:"Controls Selection Set Object Reference (by ref)"=@n2)
  ]}
 EventGraph {nodes [
   OnNeedRefresh() {is=title K2Node_CustomEvent}
   > n1 = GetManager!(execute, self) -> (then, GetManager:Python Object Handle Object Reference)
   > n2 = Execute Python Script!(
    execute
    PythonScript="import unreal\r\nmanager = unreal.resolve_python_object_handle(in_manager)\r\n\r\n## Drain surgical queue (assets we just saved/finished an op on).\r\nids = manager.consume_reconciliation()\r\n\r\n## Existing P4-watcher full-reload path.\r\ndo_full = manager.need_refresh\r\nassets = {_id: unreal.create_python_object_handle(a)\r\n              for _id, a in manager.assets.items() if _id in ids}\r\nif do_full:\r\n    manager.need_refresh = {}\r\n    assets = {_id: unreal.create_python_object_handle(a)\r\n              for _id, a in manager.assets.items()}\r\n    ids = {}"
    in_manager:Python Object Handle Object Reference=@n1
   ) -> (then, ReturnValue, assets:Map of Strings to Python Object Handle Object References, do_full:bool=~, ids:Map of Strings to Strings) {inputs [in_manager], outputs [assets,do_full,ids]}
   > Branch(Condition=@n2)
   > n3 = Branch(Condition=@n2.do_full)
   > GenerateAnimAssets!(execute, self, Assets:Map of Strings to Python Object Handle Object References=@n2.assets) -> (then)
   > Call On Refresh Assets!(execute:Exec, self:Anim Lib Manager Object Reference, Reconciliations:Map of Strings to Strings=@n2.ids) -> (then:Exec) {is=title K2Node_CallDelegate}
   Branch(execute=@n3.else, Condition=Is Not Empty!(TargetMap:Map of Strings to Strings=@n2.ids) -> (ReturnValue:bool) {is=title K2Node_CallFunction})
   > OnRefreshSubset!(execute, self, IDs:Map of Strings to Strings=@n2.ids, Assets:Map of Strings to Python Object Handle Object References=@n2.assets) -> (then)
  ]}
 FilterAssets {nodes [
   n1 = FilterAssets!() -> (then, searchParams:Map of Strings to Strings, Path:String, use_all:bool=~, filter_by_rigs:bool=~, ShowAnim:bool=~, ShowPose:bool=~) {is=title K2Node_FunctionEntry}
   > n2 = Execute Python Script!(
    execute
    PythonScript="import unreal\r\nfrom pathlib import PurePosixPath\r\nfrom AnimLibrary.core.constants import AssetType\r\n\r\nmanager = unreal.resolve_python_object_handle(in_manager)\r\n\r\nsearch_params = dict(search_params)\r\n\r\nif 'tags' in search_params:\r\n    search_params['tags'] = search_params['tags'].split(', ')\r\n\r\nif path_filter:\r\n    search_params['virtual_path'] = PurePosixPath(path_filter)\r\n\r\nif use_all or not search_params:\r\n    assets = manager.find_all(search_params, filter_by_rigs)\r\nelse:\r\n   assets = manager.find_any(search_params, filter_by_rigs)\r\n\r\nassets = [a for a in assets if (show_anim and a.metadata.asset_type is AssetType.ANIMATION)  or (show_pose and a.metadata.asset_type is AssetType.POSE)]\r\n\r\nassets = {str(a.id):unreal.create_python_object_handle(a) for a in assets}"
    in_manager=Get Manager()
    use_all:bool=@n1.use_all
    search_params:Map of Strings to Strings=@n1
    filter_by_rigs:bool=@n1.filter_by_rigs
    path_filter:String=@n1.Path
    show_anim:bool=@n1.ShowAnim
    show_pose:bool=@n1.ShowPose
   ) -> (then, ReturnValue, assets) {inputs [in_manager, use_all, search_params, filter_by_rigs, path_filter, show_anim, show_pose], outputs [assets]}
   > GenerateAnimAssets(Assets=@n2.assets)
  ]}
 GenerateAnimAssets {nodes [
   n1 = GenerateAnimAssets!() -> (then, Assets:Map of Strings to Python Object Handle Object References) {is=title K2Node_FunctionEntry}
   > Clear!(
    execute
    TargetMap:Map of Strings to Python Object Handle Object References=Get Assets() -> (Assets:Map of Strings to Python Object Handle Object References)
   ) -> (then) {is=title K2Node_CallFunction}
   > Clear(TargetMap:Map of Strings to Anim Lib Asset Object References=Get LibAssets())
   > n2 = Set Assets(Assets:Map of Strings to Python Object Handle Object References=@n1) -> (Output_Get:Map of Strings to Python Object Handle Object References)
   > n3 = "For Each Loop (Map)"!(execute:Exec, MapPin:Map of Strings to Python Object Handle Object References=@n2, BreakPin:Exec) -> (then:Exec, KeyPin:String, ValuePin:Python Object Handle Object Reference, CompletedPin:Exec)
   > AddAsset!(execute, self, handle:Python Object Handle Object Reference=@n3.ValuePin) -> (then, Asset:Anim Lib Asset Object Reference) {is=title K2Node_CallFunction}
  ]}
 GetManager {nodes [
   GetManager() {is=title K2Node_FunctionEntry}
   > n1 = Execute Python Script!(
    execute
    PythonScript="import unreal\r\nfrom AnimLibrary.core.library import Manager\r\n\r\nif in_manager is None:\r\n    lib = unreal.create_python_object_handle(Manager())\r\n    out_manager = lib\r\nelse:\r\n    out_manager = in_manager\r\n"
    in_manager=Get Manager()
   ) -> (then, ReturnValue, out_manager:Python Object Handle Object Reference) {inputs [in_manager], outputs [out_manager]}
   > n2 = Set Manager(Manager:Python Object Handle Object Reference=@n1.out_manager) -> (Output_Get:Python Object Handle Object Reference)
   > Return Node!(execute, GetManager:Python Object Handle Object Reference=@n2)
  ]}
 GetSelectionSets {nodes [
   GetSelectionSets() {is=title K2Node_FunctionEntry}
   > Clear(TargetMap:Map of Strings to Controls Selection Set Object References=Get SelectionSets())
   > n1 = Execute Python Script!(
    execute
    PythonScript="import unreal\r\n\r\nsettings = unreal.resolve_python_object_handle(settings_handle)\r\nsets = [unreal.create_python_object_handle(s.serialize())  for s in settings.selection_sets]"
    settings_handle:Python Object Handle Object Reference=Get Settings() -> (Settings:Python Object Handle Object Reference)
   ) -> (then, ReturnValue, sets:Array of Python Object Handle Object References) {inputs [settings_handle], outputs [sets]}
   > n2 = Construct Object from Class(Class=&ControlsSelectionSet_C)
   > Add(TargetMap=Get SelectionSets(), Key="[]", Value=@n2)
   > n3 = For Each Loop(Array:Array of Python Object Handle Object References=@n1.sets) -> (Array Element:Python Object Handle Object Reference)
   > n4 = Construct Object from Class(Class=&ControlsSelectionSet_C)
   > deserialize!(execute, self:Controls Selection Set Object Reference=@n4, dict:Python Object Handle Object Reference=@n3) -> (then)
   > Add(TargetMap=Get SelectionSets(), Key=Get Name!(self:Controls Selection Set Object Reference=@n4) -> (Name:String), Value=@n4)
  ]}
 GetSourcesData {nodes [
   GetSourcesData() {is=title K2Node_FunctionEntry}
   > n1 = Execute Python Script!(
    execute
    PythonScript="import unreal\r\n\r\nsettings = unreal.resolve_python_object_handle(handle)\r\nnames = [s.name for s in settings.sources]\r\npaths = [s.path for s in settings.sources]"
    handle:Python Object Handle Object Reference=Get Settings()
   ) -> (then, ReturnValue, names:Array of Strings, paths:Array of Strings) {inputs [handle], outputs [names,paths]}
   > Return Node!(execute, Names:Array of Strings=@n1.names, Paths:Array of Strings=@n1.paths)
  ]}
 GetTileSize {nodes [
   GetTileSize() {is=title K2Node_FunctionEntry}
   > n1 = Execute Python Script!(
    execute
    PythonScript="import unreal\r\n\r\nsettings = unreal.resolve_python_object_handle(settings_handle)\r\nsize = settings.size"
    settings_handle:Python Object Handle Object Reference=Get Settings()
   ) -> (then, ReturnValue, size:double=~) {inputs [settings_handle], outputs [size]}
   > Return Node!(execute, Tile Size:double=@n1.size)
  ]}
 OnRefreshAssets {nodes [OnRefreshAssets!() -> (then, Reconciliations:Map of Strings to Strings) {is=title K2Node_FunctionEntry}]}
 OnRefreshSubset {nodes [
   n1 = OnRefreshSubset!() -> (then, IDs:Map of Strings to Strings, Assets:Map of Strings to Python Object Handle Object References) {is=title K2Node_FunctionEntry}
   > n2 = "For Each Loop (Map)"(MapPin:Map of Strings to Strings=@n1) -> (ValuePin:String)
   > n3 = Switch on String!(execute:Exec, Selection:String=@n2.ValuePin) -> (Default:Exec, updated:Exec, created:Exec, removed:Exec) {pinNames [updated,created,removed]}
   n4 = Get Assets()
   Add(
    execute=@n3.created
    TargetMap:Map of Strings to Python Object Handle Object References=@n4
    Key=@n2
    Value:"Python Object Handle Object Reference (by ref)"=Find(TargetMap:Map of Strings to Python Object Handle Object References=@n1.Assets, Key=@n2) -> (Value:Python Object Handle Object Reference)
   )
   > AddAsset()
   Remove!(execute=@n3.removed, TargetMap:Map of Strings to Python Object Handle Object References=@n4, Key:"String (by ref)"=@n2) -> (then, ReturnValue:bool)
   > AddSimpleEditorNotification!(
    execute
    Text:Text="LOCTEXT(\"Animation Library\")"
    SubText:Text="LOCTEXT(\"Asset Deleted\")"
    FadeInDuration:float=0.5
    FadeOutDuration:float=2
    ExpireDuration:float=1
   ) -> (then)
   PollDirty(
    execute=@n3.updated
    self:Anim Lib Asset Object Reference=Find(TargetMap:Map of Strings to Anim Lib Asset Object References=Get LibAssets(), Key=@n2) -> (Value:Anim Lib Asset Object Reference)
   )
   Call On Refresh Assets(execute=@n2.CompletedPin, Reconciliations=@n1)
  ]}
 SetTileSize {nodes [
   n1 = SetTileSize!() -> (then, New Size:double=~) {is=title K2Node_FunctionEntry}
   > Execute Python Script!(
    execute
    PythonScript="import unreal\r\n\r\nsettings = unreal.resolve_python_object_handle(settings_handle)\r\nsettings.size = new_size"
    settings_handle=Get Settings()
    new_size:double=@n1
   ) -> (then, ReturnValue) {inputs [settings_handle, new_size]}
   > Return Node!(execute)
  ]}
 bootstrap {nodes [
   bootstrap() {is=title K2Node_FunctionEntry}
   > n1 = Sequence!(execute) -> (then_0, then_1, then_2:Exec, then_3:Exec, then_4:Exec, then_5:Exec)
   > Execute Python Script!(
    execute
    PythonScript="import sys\r\n\r\ndef clean_anim():\r\n    for m in list(sys.modules.keys()):\r\n        if 'AnimLibrary' not in m:\r\n            continue\r\n        del sys.modules[m]\r\n\r\nclean_anim()"
   ) -> (then, ReturnValue)
   Execute Python Script!(
    execute=@n1.then_1
    PythonScript="import unreal\r\nfrom AnimLibrary.core.library import Manager\r\n\r\nif in_manager is not None:\r\n    unreal.destroy_python_object_handle(in_manager)\r\nif in_settings is not None:\r\n    unreal.destroy_python_object_handle(in_settings)\r\n"
    in_manager:Python Object Handle Object Reference=Get Manager()
    in_settings:Python Object Handle Object Reference=Get Settings()
   ) -> (then, ReturnValue) {inputs [in_manager,in_settings]}
   > Set Manager()
   > Set Settings(Settings:Python Object Handle Object Reference) -> (Output_Get:Python Object Handle Object Reference)
   Clear(
    execute=@n1.then_2
    TargetArray:Array of Python Object Handle Object References=Get Sources() -> (Sources:Array of Python Object Handle Object References)
   ) {is=title K2Node_CallArrayFunction}
   n2 = GetManager(execute=@n1.then_3) {is=title K2Node_CallFunction}
   > n3 = Execute Python Script!(
    execute
    PythonScript="import unreal\r\n\r\nmanager = unreal.resolve_python_object_handle(in_manager)\r\n## manager.settings.reload()\r\nmanager.fetch()\r\nunreal.log(manager)\r\n\r\nsettings = unreal.create_python_object_handle(manager.settings)\r\n\r\nassets = {n: unreal.create_python_object_handle(a) for n, a in manager.assets.items()}\r\nsources = [unreal.create_python_object_handle(s) for s in manager.sources]"
    in_manager=@n2
   ) -> (
    then
    ReturnValue
    assets:Map of Strings to Python Object Handle Object References
    settings:Python Object Handle Object Reference
    sources:Array of Python Object Handle Object References
   ) {inputs [in_manager], outputs [assets,settings,sources]}
   > Branch(Condition=@n3)
   > n4 = Sequence!(execute) -> (then_0, then_1, then_2)
   > GenerateAnimAssets(Assets=@n3.assets) {is=title K2Node_CallFunction}
   Set Settings(execute=@n4.then_1, Settings=@n3.settings)
   Set Sources(execute=@n4.then_2, Sources:Array of Python Object Handle Object References=@n3.sources) -> (Output_Get:Array of Python Object Handle Object References)
   GetSelectionSets(execute=@n1.then_4) {is=title K2Node_CallFunction}
   n5 = Set Timer by Event(execute=@n1.then_5, Delegate=Create Event(), Time=0.5, bLooping=true, InitialStartDelay=2)
   > n6 = Set TimerHandle(TimerHandle:Timer Handle Structure=@n5) -> (Output_Get:Timer Handle Structure)
   > n7 = GetManager()
   > Execute Python Script!(
    execute
    PythonScript="import unreal\r\n\r\nmanager = unreal.resolve_python_object_handle(in_manager)\r\nmanager.refresh_timer = in_timer"
    in_manager=@n7
    in_timer:Timer Handle Structure=@n6
   ) -> (then, ReturnValue) {inputs [in_manager,in_timer]}
  ]}
 shut_down {nodes [
   n1 = GetManager()
   > n2 = Execute Python Script!(
    execute
    PythonScript="import unreal\r\n\r\nmanager = unreal.resolve_python_object_handle(in_manager)\r\nmanager.on_shut_down()"
    in_manager=@n1
   ) -> (then, ReturnValue) {inputs [in_manager]}
   Clear(execute=[@n2.then, @n9.then, @n7.No, @n5.else], TargetMap:Map of Strings to Anim Lib Asset Object References=Get LibAssets()) {is=title K2Node_CallFunction}
   > Clear(TargetMap:Map of Strings to Python Object Handle Object References=Get Assets())
   > n3 = GetEditorWorld(self=Get UnrealEditorSubsystem())
   > Clear and Invalidate Timer by Handle!(execute, WorldContextObject:Object Reference=@n3, Handle=Get TimerHandle() -> (TimerHandle:Timer Handle Structure)) -> (then)
   shut_down() {is=title K2Node_FunctionEntry}
   > Set HasDirty(HasDirty:bool=false) -> (Output_Get:bool)
   > n4 = "For Each Loop (Map)"(MapPin:Map of Strings to Anim Lib Asset Object References=Get LibAssets()) -> (ValuePin:Anim Lib Asset Object Reference)
   > Branch(Condition=Get dirty!(self:Anim Lib Asset Object Reference=@n4.ValuePin) -> (dirty:bool))
   > Set HasDirty(HasDirty=true)
   n5 = Get!(execute:Exec=@n4.CompletedPin) -> (then:Exec, else:Exec, HasDirty:bool)
   > n6 = Show Message Dialog(
    Title=MakeLiteralText(Value="LOCTEXT(\"Animation Library\")")
    Message=MakeLiteralText(Value="LOCTEXT(\"You have unsaved changes! Save before closing?\")")
    MessageType=YesNo
    DefaultValue=Yes
    MessageCategory=Warning
   )
   > n7 = Switch on EAppReturnType(Selection=@n6)
   n8 = GetManager(execute=@n7.Yes)
   > n9 = Execute Python Script(
    PythonScript="import unreal\r\n\r\nmanager = unreal.resolve_python_object_handle(in_manager)\r\nsaved = manager.save_all()\r\nunreal.log(f'Save All: persisted {saved} dirty asset(s)')"
    in_manager=@n8
   ) {inputs [in_manager]}
  ]}
]
variables [
 Assets {type="map<PythonObjectHandle>", default {}}
 LibAssets {type="map<AnimLibAsset_C>", default {}}
 Manager {type=PythonObjectHandle, default=null}
 OnRefreshAssets {default=()}
 SelectionSets {type="map<ControlsSelectionSet_C>", default {}}
 Settings {type=PythonObjectHandle, default=null}
 Sources {type="array<PythonObjectHandle>", default []}
 TimerHandle {type=TimerHandle}
]
