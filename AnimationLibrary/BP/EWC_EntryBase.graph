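# Editor Utility Widget that serves as a base list-view entry for the AnimationLibrary tool, showing an item's name as small right-aligned text.
# When a list item object is set, it casts it to EntryItemBase, stores it as ItemData and passes its name to SetName.
# Selection changes call OnSelectedDisplay, which switches the border between a blue highlight and transparent.
# SetName also contains an unconnected "- " prefix Append node.
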EditorUtilityWidgetBlueprint /AnimationLibrary/BP/EWC_EntryBase
graphs [
 EventGraph {nodes [
   n1 = Event On Item Selection Changed!() -> (OutputDelegate, then, bIsSelected:bool) {is=title K2Node_Event}
   > OnSelectedDisplay!(execute, self, selectionState:bool=@n1.bIsSelected) -> (then)
   n2 = Event OnListItemObjectSet()
   > n3 = Cast To EntryItemBase!(execute, Object=@n2.ListItemObject) -> (then, CastFailed, AsEntry Item Base:Entry Item Base Object Reference) {is=title K2Node_DynamicCast}
   > n4 = Set ItemData(ItemData:Entry Item Base Object Reference=@n3) -> (Output_Get:Entry Item Base Object Reference)
   > SetName!(execute, self, Name:String=Get name!(self:Entry Item Base Object Reference=@n4) -> (name:String)) -> (then)
  ]}
 OnSelectedDisplay {nodes [
   n1 = OnSelectedDisplay!() -> (then, selectionState:bool=~) {is=title K2Node_FunctionEntry}
   > SetBrushColor(
    self=Get border_entry() -> (border_entry:Border Object Reference)
    InBrushColor=SelectColor(A=(R=0,G=0.1608,B=0.7451,A=1), B=(R=0,G=0,B=0,A=0), bPickA=@n1)
   )
  ]}
 SetName {nodes [
   n1 = SetName!() -> (then, Name:String) {is=title K2Node_FunctionEntry}
   > Set Text(Text="To Text (String)"(InString=@n1), self=Get item_name() -> (item_name:Common Text Object Reference))
   Append(A="- ", B=@n1)
  ]}
]
variables [ItemData {type=EntryItemBase_C, default=null}]
widgets {
 info {rootWidgetClass=&Border, widgetCount=3}
 widgets [
  border_entry {
   widgetClassPath=&Border
   bIsVariable=true
   widget=border_entry
   properties {horizontalAlignment=HAlign_Right, verticalAlignment=VAlign_Top, brushColor=<rgba 1 1 1 0>}
  }
  hb_folder {
   widgetClassPath=&HorizontalBox
   bIsVariable=true
   parent=border_entry
   widget=hb_folder
   properties {bOverride_Cursor=true, visibility=Visible}
   slot {class=BorderSlot, properties {horizontalAlignment=HAlign_Right, verticalAlignment=VAlign_Top}}
  }
  item_name {
   widgetClassPath=&CommonTextBlock
   bIsVariable=true
   parent=hb_folder
   widget=item_name
   properties {
    justification=Right
    text=folder
    visibility=Visible
    wrappingPolicy=AllowPerCharacterWrapping
    colorAndOpacity {colorUseRule=UseColor_Specified, specifiedColor=<rgba 0.807 0.807 0.807 1>}
    font {
     bForceMonospaced=false
     bMaterialIsStencil=false
     fontMaterial=null
     fontObject=&Roboto
     letterSpacing=0
     monospacedWidth=1
     size=9
     skewAmount=0
     typefaceFontName=Regular
     outlineSettings {
      bApplyOutlineToDropShadows=false
      bMiteredCorners=false
      bSeparateFillAlpha=false
      outlineMaterial=null
      outlineSize=0
      outlineColor=<rgba 0 0 0 0.5>
     }
    }
    shadowColorAndOpacity=<rgba 0 0 0 0.5>
    shadowOffset=<0.01 0.01>
    strikeBrush {
     bIsDynamicallyLoaded=false
     drawAs=Image
     imageType=NoImage
     mirroring=NoMirror
     resourceName=None
     resourceObject=null
     tiling=NoTile
     imageSize=<32 32>
     margin=<ltrb 0 0 0 0>
     outlineSettings {
      bUseBrushTransparency=false
      roundingType=HalfHeightRadius
      width=0
      color {colorUseRule=UseColor_Specified, specifiedColor=<rgba 0 0 0 0>}
      cornerRadii=<0 0 0 0>
     }
     tintColor {colorUseRule=UseColor_Specified, specifiedColor=<rgba 0.807 0.807 0.807 1>}
     uVRegion {bIsValid=false, max=<0 0>, min=<0 0>}
    }
   }
   slot {class=HorizontalBoxSlot, properties {horizontalAlignment=HAlign_Right, verticalAlignment=VAlign_Center}}
  }
 ]
}
