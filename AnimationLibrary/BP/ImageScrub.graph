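# Editor utility widget for the animation library that previews an image-sequence media source and scrubs it on mouse hover.
# On construct it duplicates the template player and texture under GUID names, giving each instance its own paused player with large cache windows.
# OnSetSource opens the ImgMediaSource, rewinding once loaded.
# While the hover-set isPlaying flag is true, mouse X maps to a time across the duration, seeking only when that value changes.
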EditorUtilityWidgetBlueprint /AnimationLibrary/BP/ImageScrub
graphs [
 EventGraph {nodes [
   n1 = OnSetSource!() -> (OutputDelegate, then, mediasource:Media Source Object Reference) {is=title K2Node_CustomEvent}
   > n2 = Set mediasource(mediasource:Media Source Object Reference=@n1.mediasource) -> (Output_Get:Media Source Object Reference)
   n3 = Cast To ImgMediaSource!(execute=[@n16.then, @n2.then], Object=@n2) -> (then, CastFailed, AsImg Media Source:Img Media Source Object Reference) {is=title K2Node_DynamicCast}
   > n4 = Is Valid(InputObject=@n3)
   PrintString(execute=@n4.Is Not Valid, InString=Invalid MediaSource, Key=AnimLibrary)
   > PrintString(
    InString=Append!(A=GetDisplayName(Object=@n3), B="\r\n", C:String=GetSequencePath!(self:Img Media Source Object Reference=@n3) -> (ReturnValue:String)) -> (ReturnValue)
    Key=AnimLibrary
   )
   n5 = Get mediaplayer() -> (mediaplayer:Media Player Object Reference)
   OpenSource!(execute=@n4.Is Valid, self:Media Player Object Reference=@n5, MediaSource:Media Source Object Reference=@n3) -> (then, ReturnValue:bool)
   UpdateResource() {is=title K2Node_CustomEvent}
   n6 = Get mediatexture() -> (mediatexture:Media Texture Object Reference)
   > UpdateResource(self:Media Texture Object Reference=@n6) {is=title K2Node_CallFunction}
   n7 = OnMediaOpened!() -> (OutputDelegate, then, OpenedUrl:String) {is=title K2Node_CustomEvent}
   > Seek!(
    execute
    self:Media Player Object Reference=@n5
    Time:"Timespan Structure (by ref)"=MakeTimespan!(Days:int=0, Hours:int=0, Minutes:int=0, Seconds:int=0, Milliseconds:int=0) -> (ReturnValue:Timespan Structure)
   ) -> (then, ReturnValue:bool)
   n8 = OnMediaFailed!() -> (OutputDelegate, then, FailedUrl:String) {is=title K2Node_CustomEvent}
   Set isPlaying(execute=[@n8.then, @n19.then], isPlaying:bool=false) -> (Output_Get:bool)
   > Set members in Slate Brush!(
    execute:Exec
    StructRef:"Slate Brush Structure (by ref)"=Get Brush!(self:Image Object Reference=Get image() -> (image:Image Object Reference)) -> (Brush:Slate Brush Structure)
    ResourceObject:Object Reference
   ) -> (then:Exec, StructOut:"Slate Brush Structure (by ref)") {is=title K2Node_SetFieldsInStruct}
   Event Construct()
   > n9 = Sequence!(execute) -> (then_0, then_1, then_2:Exec, then_3:Exec, then_4:Exec)
   n10 = "To String (Guid)"!(InGuid:"Guid Structure (by ref)"=NewGuid!() -> (ReturnValue:Guid Structure)) -> (ReturnValue:String)
   > n11 = DuplicateObject!(
    execute
    SourceObject:Object Reference=&/ThumbPlayer
    Outer:Object Reference
    Name:Name=String To Name(InString=Append!(A=Player, B=@n10) -> (ReturnValue))
   ) -> (then, ReturnValue:Object Reference)
   > n12 = Cast To MediaPlayer!(execute, Object=@n11) -> (then, CastFailed, AsMedia Player:Media Player Object Reference) {is=title K2Node_DynamicCast}
   > Set mediaplayer(mediaplayer:Media Player Object Reference=@n12) -> (Output_Get:Media Player Object Reference)
   n13 = DuplicateObject(execute=@n9.then_1, SourceObject=&/ThumbPlayer_Video, Name=String To Name(InString=Append(A=texture, B=@n10)))
   > n14 = Cast To MediaTexture!(execute, Object=@n13) -> (then, CastFailed, AsMedia Texture:Media Texture Object Reference) {is=title K2Node_DynamicCast}
   > Set mediatexture(mediatexture:Media Texture Object Reference=@n14) -> (Output_Get:Media Texture Object Reference)
   Set members in Slate Brush(execute=@n9.then_3, StructRef=Get Brush(self=Get image()), ResourceObject=Get mediatexture())
   Close(execute=@n9.then_4, self:Media Player Object Reference=@n5)
   > Bind Event to On Media Opened(self:Media Player Object Reference=@n5, Delegate=@n7) {is=title K2Node_AddDelegate}
   > Bind Event to On Media Open Failed(self:Media Player Object Reference=@n5, Delegate=@n8) {is=title K2Node_AddDelegate}
   n15 = MakeTimespan(Days=0, Hours=0, Minutes=10, Seconds=0, Milliseconds=0)
   > Set CacheAhead!(execute, CacheAhead:Timespan Structure=@n15, self:Media Player Object Reference=@n5) -> (then, Output_Get:Timespan Structure)
   > Set CacheBehind!(execute, CacheBehind:Timespan Structure=@n15, self:Media Player Object Reference=@n5) -> (then, Output_Get:Timespan Structure)
   > Set PlayOnOpen!(execute, PlayOnOpen:bool=false, self:Media Player Object Reference=@n5) -> (then, Output_Get:bool)
   > SetMediaPlayer!(execute, self:Media Texture Object Reference=@n6, NewMediaPlayer:Media Player Object Reference=@n5) -> (then)
   > n16 = SetRate!(execute, self:Media Player Object Reference=@n5, Rate:float=0) -> (then, ReturnValue:bool)
   Event OnMouseEnter()
   n17 = Get mediasource() -> (mediasource:Media Source Object Reference)
   > n18 = Branch(Condition=AND Boolean(A=Validate!(self:Media Source Object Reference=@n17) -> (ReturnValue:bool), B=Get isPlaying() -> (isPlaying:bool)))
   Set isPlaying(execute=@n18.else, isPlaying=true)
   n19 = Event OnMouseLeave!() -> (OutputDelegate, then, MouseEvent:"Pointer Event Structure (by ref)") {is=title K2Node_Event}
   Event PreConstruct()
   Event Tick()
  ]}
 OnMouseMove {nodes [
   n1 = OnMouseMove!() -> (then, MyGeometry:Geometry Structure, MouseEvent:"Pointer Event Structure (by ref)") {is=title K2Node_FunctionEntry}
   n2 = Get Screen Space Position!(Input:"Pointer Event Structure (by ref)"=@n1.MouseEvent) -> (ReturnValue:Vector 2D Structure) {is=title K2Node_CallFunction}
   n3 = GetTickSpaceGeometry!(self:Widget Object Reference=Self-Reference()) -> (ReturnValue:"Geometry Structure (by ref)")
   n4 = GetAbsolutePosition!(Geometry:"Geometry Structure (by ref)"=@n3) -> (ReturnValue:Vector 2D Structure)
   n5 = BreakVector2D(InVec=GetAbs2D!(A:Vector 2D Structure=vector2d - vector2d(A:Vector 2D Structure=@n2, B:Vector 2D Structure=@n4) -> (ReturnValue:Vector 2D Structure) {is=title K2Node_PromotableOperator}) -> (ReturnValue:Vector 2D Structure))
   n6 = GetAbsoluteSize!(Geometry:"Geometry Structure (by ref)"=@n3) -> (ReturnValue:Vector 2D Structure)
   n7 = float / float(A=@n5, B=BreakVector2D(InVec=@n6))
   n8 = "Clamp (Float)"(Value=@n7, Min=0, Max=1)
   n9 = Get mediaplayer()
   n10 = Get Total Milliseconds!(A:Timespan Structure=GetDuration!(self:Media Player Object Reference=@n9) -> (ReturnValue:Timespan Structure)) -> (ReturnValue:double) {is=title K2Node_CallFunction}
   n11 = Round(A=float * float(A=@n8, B=float - float(A=@n10, B=1)) {is=title K2Node_PromotableOperator}) {is=title K2Node_CallFunction}
   n12 = Get lastSeek() -> (lastSeek:int=0)
   > Branch(Condition=AND Boolean(A="Not Equal (Integer)"(A=@n12, B=@n11), B=Get isPlaying()))
   > n13 = Branch(Condition=false)
   > PrintString(InString=Append!(
     A="Mouse: "
     B="To String (Vector2d)"!(InVec:Vector 2D Structure=@n2) -> (ReturnValue:String)
     C:String=" Widget: "
     D:String="To String (Vector2d)"(InVec=@n4)
     E:String=" Size: "
     F:String="To String (Vector2d)"(InVec=@n6)
     G:String=" Delta: "
     H:String="To String (Float)"(InDouble=@n5)
    ) -> (ReturnValue))
   > n14 = PrintString(InString=Append!(
     A="%: "
     B="To String (Float)"(InDouble=@n7)
     C=clamped: %
     D="To String (Float)"(InDouble=@n8)
     E=frame:
     F="To String (Integer)"(InInt=@n11)
    ) -> (ReturnValue))
   n15 = Set lastSeek(execute=[@n14.then, @n13.else], lastSeek:int=@n11) -> (Output_Get:int=0)
   > Seek(self=@n9, Time=MakeTimespan(Days=0, Hours=0, Minutes=0, Seconds=0, Milliseconds=@n15))
   > Return Node!(execute, ReturnValue:Event Reply Structure=Handled())
   Get image()
  ]}
]
variables [
 isPlaying {type=boolean}
 lastSeek {type=integer}
 mediaplayer {type=MediaPlayer, default=null}
 mediasource {type=MediaSource, default=null}
 mediatexture {type=MediaTexture, default=null}
]
widgets {
 info {rootWidgetClass=&Image, widgetCount=1}
 widgets [image {
   widgetClassPath=&Image
   bIsVariable=true
   widget=image
   properties {brush {
     bIsDynamicallyLoaded=false
     drawAs=RoundedBox
     imageType=NoImage
     mirroring=NoMirror
     resourceName=None
     resourceObject=null
     tiling=NoTile
     imageSize=<256 256>
     margin=<ltrb 0 0 0 0>
     outlineSettings {
      bUseBrushTransparency=false
      roundingType=FixedRadius
      width=0
      color {colorUseRule=UseColor_Specified, specifiedColor=<rgba 0 0 0 0>}
      cornerRadii=<8 8 8 8>
     }
     tintColor {colorUseRule=UseColor_Specified, specifiedColor=<rgba 1 1 1 1>}
     uVRegion {bIsValid=false, max=<0 0>, min=<0 0>}
    }}
  }]
}
