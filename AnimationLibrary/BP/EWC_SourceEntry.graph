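# Editor utility list-entry widget for one animation library source in the Settings tab, showing name, path, type and a default-source icon.
# List item data supplies a Python object handle, read and modified through embedded Python scripts.
# Validated path edits are saved; validated name edits call set_default_repository rather than renaming.
# A right-click menu offers set default, delete and Explorer opening; changes notify the parent library via OnSourceChanged.
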EditorUtilityWidgetBlueprint /AnimationLibrary/BP/EWC_SourceEntry
graphs [
 EventGraph {nodes [
   n1 = "OnTextCommitted (Editable Text) (ItemPath)"!() -> (OutputDelegate, then, Text:"Text (by ref)", CommitMethod:ETextCommit Enum=Default) {is=title K2Node_ComponentBoundEvent}
   n2 = "To String (Text)"(InText=@n1.Text)
   n3 = Get SettingsTab() -> (SettingsTab:EUW Anim Library Tab Settings Object Reference)
   > n4 = validateSourceInput!(execute, self:EUW Anim Library Tab Settings Object Reference=@n3, name:String=__temp, path:String=@n2) -> (then, valid:bool)
   > Branch(Condition=@n4)
   n5 = Get AssetHandle() -> (AssetHandle:Python Object Handle Object Reference)
   n6 = Get library!(self:EUW Anim Library Object Reference=Get parent!(self:EUW Anim Library Tab Settings Object Reference=Get SettingsTab()) -> (parent:EUW Anim Library Object Reference)) -> (library:Anim Lib Manager Object Reference)
   n7 = Get Settings!(self:Anim Lib Manager Object Reference=@n6) -> (Settings:Python Object Handle Object Reference)
   > n8 = Execute Python Script!(
    execute
    PythonScript="import unreal\r\n\r\nsrc = unreal.resolve_python_object_handle(src_handle)\r\nsrc.path = path\r\nsettings = unreal.resolve_python_object_handle(settings_handle)\r\nsettings.save()"
    src_handle:Python Object Handle Object Reference=@n5
    path:String=@n2
    settings_handle:Python Object Handle Object Reference=@n7
   ) -> (then, ReturnValue) {inputs [src_handle, path, settings_handle]}
   OnSourceChanged(execute=[@n8.then, @n12.then], self:EUW Anim Library Object Reference=Get parent(self=Get SettingsTab()))
   n9 = "OnTextCommitted (Editable Text) (ItemName)"!() -> (OutputDelegate, then, Text:"Text (by ref)", CommitMethod:ETextCommit Enum=Default) {is=title K2Node_ComponentBoundEvent}
   n10 = "To String (Text)"(InText=@n9.Text)
   > n11 = validateSourceInput(self=@n3, name=@n10, path=__temp)
   > Branch(Condition=@n11)
   > n12 = Execute Python Script!(
    execute
    PythonScript="import unreal\r\n\r\nmanager = unreal.resolve_python_object_handle(in_manager)\r\nsrc = unreal.resolve_python_object_handle(src_handle)\r\n\r\nmanager.set_default_repository(src)"
    src_handle=@n5
    name:String=@n10
    settings_handle=@n7
    in_manager:Python Object Handle Object Reference=Get Manager!(self:Anim Lib Manager Object Reference=@n6) -> (Manager:Python Object Handle Object Reference)
   ) -> (then, ReturnValue) {inputs [src_handle, name, settings_handle, in_manager]}
   n13 = onSetAsDefault!() -> (OutputDelegate, then, Context Item:Editor Utility Widget Object Reference, Slot Widget:Editor Utility Widget Object Reference) {is=title K2Node_CustomEvent}
   n14 = Get Settings(self=Get ParentLibrary() -> (ParentLibrary:Anim Lib Manager Object Reference))
   n15 = Get AssetHandle()
   > n16 = Execute Python Script!(
    execute
    PythonScript="import unreal\r\n\r\nsettings = unreal.resolve_python_object_handle(settings_handle)\r\nsrc = unreal.resolve_python_object_handle(src_handle)\r\n\r\nsettings.default_source = src\r\n"
    settings_handle=@n14
    src_handle=@n15
   ) -> (then, ReturnValue) {inputs [settings_handle, src_handle]}
   Close(
    execute=[@n16.then, @n18.then]
    self:Menu Anchor Object Reference=Get ContextMenuAnchor() -> (ContextMenuAnchor:Menu Anchor Object Reference)
   )
   > OnSourceChanged(self=Get parent(self=Get SettingsTab()))
   n17 = onDeleteSource!() -> (OutputDelegate, then, Context Item:Editor Utility Widget Object Reference, Slot Widget:Editor Utility Widget Object Reference) {is=title K2Node_CustomEvent}
   > n18 = Execute Python Script!(
    execute
    PythonScript="import unreal\r\n\r\nmanager = unreal.resolve_python_object_handle(in_manager)\r\nsrc = unreal.resolve_python_object_handle(src_handle)\r\n\r\nmanager.remove_source(src)"
    settings_handle=@n14
    src_handle=@n15
    in_manager:Python Object Handle Object Reference=Get Manager(self=Get ParentLibrary())
   ) -> (then, ReturnValue) {inputs [settings_handle, src_handle, in_manager]}
   n19 = onOpenPath!() -> (OutputDelegate, then, Context Item:Editor Utility Widget Object Reference, Slot Widget:Editor Utility Widget Object Reference) {is=title K2Node_CustomEvent}
   n20 = Get AssetHandle()
   > Is Valid(InputObject=@n20)
   > Execute Python Script!(
    execute
    PythonScript="import os\r\nimport subprocess\r\nimport unreal\r\n\r\nsrc = unreal.resolve_python_object_handle(src_handle)\r\nsubprocess.Popen(f'explorer \"{os.path.normpath(src.path)}\"')"
    src_handle=@n20
   ) -> (then, ReturnValue) {inputs [src_handle]}
   Event Construct()
   > n21 = Create Widget!(
    execute:Exec
    WorldContextObject:Object Reference
    Class:User Widget Class Reference=/Game/NeverCook/UI/ContextMenu/EUW_ContextMenu.EUW_ContextMenu_C
    OwningPlayer:Player Controller Object Reference
   ) -> (then:Exec, ReturnValue:EUW Context Menu Object Reference)
   > n22 = Set ContextMenu(ContextMenu:EUW Context Menu Object Reference=@n21) -> (Output_Get:EUW Context Menu Object Reference)
   > Set OwnerWidget!(execute, OwnerWidget:Widget Object Reference, self:EUW Context Menu Object Reference=@n22) -> (then, Output_Get:Widget Object Reference)
   > n23 = Sequence!(execute) -> (then_0, then_1, then_2:Exec)
   > Add!(
    execute
    self:EUW Context Menu Object Reference=@n22
    Event:"Delegate (by ref)"=@n13
    Category:String=Edit
    Name:String=Set As Default
    Icon:Texture 2D Object Reference=/AnimationLibrary/resources/icons/tagsolid
    ItemWidget:User Widget Class Reference=/Game/NeverCook/UI/ContextMenu/EUW_ContextMenu_Button.EUW_ContextMenu_Button_C
    Tooltip:String
   ) -> (then, Context Menu Item:Widget Object Reference, ContainerItem:User Widget Object Reference) {is=title K2Node_CallFunction}
   Add(
    execute=@n23.then_1
    self=@n22
    Event=@n17
    Category=Edit
    Name=Delete
    Icon=/AnimationLibrary/resources/icons/trashsolid
    ItemWidget=&EUW_ContextMenu_Button_C
   )
   Add(
    execute=@n23.then_2
    self=@n22
    Event=@n19
    Category=Open
    Name=Open Path in Explorer
    Icon=/AnimationLibrary/resources/icons/folderopensolid
    ItemWidget=&EUW_ContextMenu_Button_C
   )
   n24 = Event OnListItemObjectSet()
   n25 = Event On Item Selection Changed!() -> (OutputDelegate, then, bIsSelected:bool) {is=title K2Node_Event}
   OnSelectedDisplay!(execute=[@n24.then, @n25.then], self, selectionState:bool=@n25.bIsSelected) -> (then)
   > SetVisibility(self=Get icon() -> (icon:Image Object Reference), InVisibility=Hidden)
   > n26 = Cast To SourceEntryItemData!(execute, Object=@n24.ListItemObject) -> (then, CastFailed, AsSource Entry Item Data:Source Entry Item Data Object Reference) {is=title K2Node_DynamicCast}
   > Set Item(Item:Source Entry Item Data Object Reference=@n26) -> (Output_Get:Source Entry Item Data Object Reference)
   > Set AssetHandle(AssetHandle:Python Object Handle Object Reference=Get item handle!(self:Source Entry Item Data Object Reference=@n26) -> (item handle:Python Object Handle Object Reference)) -> (Output_Get:Python Object Handle Object Reference)
   > Set ParentLibrary(ParentLibrary:Anim Lib Manager Object Reference=Get library(self:Source Entry Item Data Object Reference=@n26)) -> (Output_Get:Anim Lib Manager Object Reference)
   > Set SettingsTab(SettingsTab:EUW Anim Library Tab Settings Object Reference=Get parentTab!(self:Source Entry Item Data Object Reference=@n26) -> (parentTab:EUW Anim Library Tab Settings Object Reference)) -> (Output_Get:EUW Anim Library Tab Settings Object Reference)
   > n27 = GetSourceNameAndPath!(execute, self) -> (then, Name:String, Path:String, default:bool, Source Type:String)
   > SetItemName!(execute, self, Name:String=@n27) -> (then)
   > Set Text(
    Text="To Text (String)"(InString=@n27.Path)
    self:Text Box Object Reference=Get ItemPath() -> (ItemPath:Editor Utility Editable Text Box Object Reference)
   )
   > n28 = Set Text(Text="To Text (String)"(InString=@n27.Source Type), self:Text Object Reference=Get ItemType() -> (ItemType:Text Object Reference))
   SetVisibility(execute=[@n28.then, @n30.then], self=Get icon(), InVisibility=Select(Option 0=Hidden, Option 1=Visible, Index=@n27.default))
   n29 = SelectString(A=Default Source, bPickA=@n27.default)
   n30 = Set ToolTipText!(
    execute
    ToolTipText:Text="To Text (String)"(InString=Append!(
      A="Source: "
      B=@n27
      C:String="\r\nPath: "
      D:String=@n27.Path
      E:String="\r\nType: "
      F:String=@n27.Source Type
      G:String="\r\n"
      H:String=@n29
     ) -> (ReturnValue))
    self:Widget Object Reference
   ) -> (then, Output_Get:Text)
  ]}
 GetSourceNameAndPath {nodes [
   GetSourceNameAndPath() {is=title K2Node_FunctionEntry}
   > n1 = Execute Python Script!(
    execute
    PythonScript="import unreal\r\n\r\nobj = unreal.resolve_python_object_handle(handle)\r\nname = obj.name\r\npath = obj.path  ## works via the new path property added to AssetRepository\r\nis_default = obj.is_default\r\nsrc_type = obj.vcs.descriptor_tag  ## 'PerforceSource' | 'UESource' | 'LocalSource' | etc."
    handle:Python Object Handle Object Reference=Get AssetHandle()
   ) -> (then, ReturnValue, name:String, path:String, is_default:bool=~, srcType:String) {inputs [handle], outputs [name,path,is_default,srcType]}
   > Return Node!(execute, Name:String=@n1.name, Path:String=@n1.path, default:bool=@n1.is_default, Source Type:String=@n1.srcType)
  ]}
 OnMouseButtonDown {nodes [
   n1 = OnMouseButtonDown!() -> (then, MyGeometry:Geometry Structure, MouseEvent:"Pointer Event Structure (by ref)") {is=title K2Node_FunctionEntry}
   > n2 = Branch(Condition=Is Mouse Button Down!(Input:"Pointer Event Structure (by ref)"=@n1.MouseEvent, MouseButton:Key Structure=RightMouseButton) -> (ReturnValue:bool))
   > Open!(execute, self:Menu Anchor Object Reference=Get ContextMenuAnchor(), bFocusMenu:bool=true) -> (then)
   > Return Node!(execute, ReturnValue:Event Reply Structure=Handled())
   Return Node(execute=@n2.else, ReturnValue=Unhandled())
  ]}
 OnSelectedDisplay {nodes [
   n1 = OnSelectedDisplay!() -> (then, selectionState:bool=~) {is=title K2Node_FunctionEntry}
   > SetBrushColor(
    self=Get asset_border() -> (asset_border:Border Object Reference)
    InBrushColor=SelectColor(A=(R=0,G=0.1608,B=0.7451,A=1), B=(R=0,G=0,B=0,A=0), bPickA=@n1)
   )
  ]}
 On_ContextMenuAnchor_GetUserMenuContent {nodes [
   On_ContextMenuAnchor_GetUserMenuContent() {is=title K2Node_FunctionEntry}
   n1 = Get ContextMenu() -> (ContextMenu:EUW Context Menu Object Reference)
   > BuildContent(self:EUW Context Menu Object Reference=@n1)
   > Return Node(ReturnValue:User Widget Object Reference=@n1)
  ]}
 SetItemName {nodes [
   n1 = SetItemName!() -> (then, Name:String) {is=title K2Node_FunctionEntry}
   > Set Text(
    Text="To Text (String)"(InString=@n1)
    self:Text Box Object Reference=Get ItemName() -> (ItemName:Editor Utility Editable Text Box Object Reference)
   )
  ]}
]
variables [
 AssetHandle {type=PythonObjectHandle, default=null}
 ContextMenu {type=EUW_ContextMenu_C, default=null}
 Item {type=SourceEntryItemData_C, default=null}
 ParentLibrary {type=AnimLibManager_C, default=null}
 SettingsTab {type=EUW_AnimLibrary_Tab_Settings_C, default=null}
]
widgets {
 info {rootWidgetClass=&Border, widgetCount=8}
 widgets [
  asset_border {
   widgetClassPath=&Border
   bIsVariable=true
   widget=asset_border
   properties {
    background {
     bIsDynamicallyLoaded=false
     drawAs=Border
     imageType=NoImage
     mirroring=NoMirror
     resourceName=None
     resourceObject=null
     tiling=NoTile
     imageSize=<32 32>
     margin=<ltrb 1 1 1 1>
     outlineSettings {
      bUseBrushTransparency=false
      roundingType=HalfHeightRadius
      width=0
      color {colorUseRule=UseColor_Specified, specifiedColor=<rgba 0 0 0 0>}
      cornerRadii=<0 0 0 0>
     }
     tintColor {colorUseRule=UseColor_Specified, specifiedColor=<rgba 1 1 1 1>}
     uVRegion {bIsValid=false, max=<0 0>, min=<0 0>}
    }
    brushColor=<rgba 0.026 0.026 0.026 1>
    padding=<ltrb 6 6 0 6>
   }
  }
  HorizontalBox_0 {
   widgetClassPath=&HorizontalBox
   parent=asset_border
   widget=HorizontalBox_0
   slot {class=BorderSlot, properties {padding=<ltrb 6 6 0 6>}}
  }
  icon {
   widgetClassPath=&Image
   bIsVariable=true
   parent=HorizontalBox_0
   widget=icon
   properties {
    bIsEnabled=false
    visibility=Hidden
    brush {
     bIsDynamicallyLoaded=false
     drawAs=Image
     imageType=FullColor
     mirroring=NoMirror
     resourceName=None
     resourceObject=null
     tiling=NoTile
     imageSize=<16 16>
     margin=<ltrb 0 0 0 0>
     outlineSettings {
      bUseBrushTransparency=false
      roundingType=HalfHeightRadius
      width=0
      color {colorUseRule=UseColor_Specified, specifiedColor=<rgba 0 0 0 0>}
      cornerRadii=<0 0 0 0>
     }
     tintColor {colorUseRule=UseColor_Specified, specifiedColor=<rgba 1 1 1 1>}
     uVRegion {bIsValid=false, max=<0 0>, min=<0 0>}
    }
   }
   slot {class=HorizontalBoxSlot, properties {horizontalAlignment=HAlign_Center, verticalAlignment=VAlign_Center}}
  }
  AssetTile {
   widgetClassPath=&VerticalBox
   parent=HorizontalBox_0
   widget=AssetTile
   properties {bOverride_Cursor=true, visibility=Visible}
   slot {class=HorizontalBoxSlot, properties {padding=<ltrb 0 0 4 0>, size {sizeRule=Fill, value=1}}}
  }
  ItemName {
   widgetClassPath=&EditorUtilityEditableTextBox
   bIsVariable=true
   parent=AssetTile
   widget=ItemName
   properties {bOverride_Cursor=true, clipping=ClipToBounds}
   slot {class=VerticalBoxSlot, properties {verticalAlignment=VAlign_Center, size {sizeRule=Fill, value=1}}}
  }
  ItemPath {
   widgetClassPath=&EditorUtilityEditableTextBox
   bIsVariable=true
   parent=AssetTile
   widget=ItemPath
   properties {bOverride_Cursor=true, clipping=ClipToBounds}
   slot {class=VerticalBoxSlot, properties {verticalAlignment=VAlign_Center, size {sizeRule=Fill, value=1}}}
  }
  ItemType {
   widgetClassPath=&TextBlock
   bIsVariable=true
   parent=AssetTile
   widget=ItemType
   properties {
    colorAndOpacity {colorUseRule=UseColor_Specified, specifiedColor=<rgba 1 1 1 0.5>}
    font {
     bForceMonospaced=false
     bMaterialIsStencil=false
     fontMaterial=null
     fontObject=&Roboto
     letterSpacing=0
     monospacedWidth=1
     size=9
     skewAmount=0
     typefaceFontName=Italic
     outlineSettings {
      bApplyOutlineToDropShadows=false
      bMiteredCorners=false
      bSeparateFillAlpha=false
      outlineMaterial=null
      outlineSize=0
      outlineColor=<rgba 0 0 0 1>
     }
    }
   }
   slot {class=VerticalBoxSlot, properties {verticalAlignment=VAlign_Center, size {sizeRule=Fill, value=1}}}
  }
  ContextMenuAnchor {
   widgetClassPath=/Script/UMG.MenuAnchor
   bIsVariable=true
   parent=AssetTile
   widget=ContextMenuAnchor
   properties {menuClass=&EUW_ContextMenu_C, placement=MenuPlacement_Center}
   slot {class=VerticalBoxSlot, properties {verticalAlignment=VAlign_Center}}
  }
 ]
}
