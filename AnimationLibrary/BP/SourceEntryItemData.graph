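# SourceEntryItemData is an editor utility Blueprint that acts as a data object for a source entry item in the animation library tool.
# It holds a Python object handle for the item, plus references to the AnimLibManager and the owning Animation Library settings tab widget.
# It has no graph logic; its empty event graph leaves it as a plain container.
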EditorUtilityBlueprint /AnimationLibrary/BP/SourceEntryItemData
graphs [EventGraph {links [], nodes []}]
variables [
 item handle {type=PythonObjectHandle, default=null}
 library {type=AnimLibManager_C, default=null}
 parentTab {type=EUW_AnimLibrary_Tab_Settings_C, default=null}
]
