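# Editor utility dialog widget that serves as the Settings tab of the Animation Library tool, covering options and source root locations.
# Checkboxes read and write key reducer, binary format and missing-channel options on a Python settings object through Execute Python Script.
# Source handles from the library manager fill a list whose double-click opens Explorer.
# The validated add-source form and refresh button are collapsed and disabled.
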EditorUtilityWidgetBlueprint /AnimationLibrary/BP/EUW_AnimLibrary_Tab_Settings
graphs [
 EventGraph {nodes [
   n1 = "On Check State Changed (chk_use_reducer)"!() -> (OutputDelegate, then, bIsChecked:bool) {is=title K2Node_ComponentBoundEvent}
   n2 = Get Settings!(self:Anim Lib Manager Object Reference=Get library!(self:EUW Anim Library Object Reference=Get parent() -> (parent:EUW Anim Library Object Reference)) -> (library:Anim Lib Manager Object Reference)) -> (Settings:Python Object Handle Object Reference)
   > Execute Python Script!(
    execute
    PythonScript="import unreal\r\n\r\nobj = unreal.resolve_python_object_handle(handle)\r\nobj.use_reducer = is_checked"
    handle:Python Object Handle Object Reference=@n2
    is_checked:bool=@n1.bIsChecked
   ) -> (then, ReturnValue) {inputs [handle,is_checked]}
   n3 = "On Check State Changed (chk_as_binary)"!() -> (OutputDelegate, then, bIsChecked:bool) {is=title K2Node_ComponentBoundEvent}
   > Execute Python Script(
    PythonScript="import unreal\r\n\r\nobj = unreal.resolve_python_object_handle(handle)\r\nobj.as_binary = is_checked"
    handle=@n2
    is_checked=@n3.bIsChecked
   ) {inputs [handle,is_checked]}
   "On Clicked (AddSourceBtn)"() {is=title K2Node_ComponentBoundEvent}
   n4 = "To String (Text)"(InText=Get Text(self=Get edt_NewPath() -> (edt_NewPath:Editor Utility Editable Text Box Object Reference)))
   n5 = "To String (Text)"(InText=Get Text(self=Get edt_NewName() -> (edt_NewName:Editor Utility Editable Text Box Object Reference)))
   > n6 = validateSourceInput!(execute, self, name:String=@n4, path:String=@n5) -> (then, valid:bool)
   > n7 = Branch(Condition=@n6)
   n8 = Get library(self=Get parent())
   > Execute Python Script!(
    execute
    PythonScript="import unreal\r\n\r\nfrom AnimLibrary.core.repository_factory import build_local\r\n\r\nmanager = unreal.resolve_python_object_handle(in_manager)\r\nnew_src = build_local(name=new_source_name, path=new_source_path)\r\nmanager.add_source(new_src, as_default=True)"
    handle=Get Settings(self=@n8)
    new_source_name:String=@n4
    new_source_path:String=@n5
    in_manager:Python Object Handle Object Reference=Get Manager!(self:Anim Lib Manager Object Reference=@n8) -> (Manager:Python Object Handle Object Reference)
   ) -> (then, ReturnValue) {inputs [handle, new_source_name, new_source_path, in_manager]}
   > OnSourceChanged(self:EUW Anim Library Object Reference=Get parent())
   > Set Text(self:Text Box Object Reference=Get edt_NewPath())
   > Set Text(self=Get edt_NewName())
   Show Message Dialog(
    execute=@n7.else
    Title=MakeLiteralText(Value="LOCTEXT(\"Adding new source\")")
    Message=MakeLiteralText(Value="LOCTEXT(\"Invalid input for new source, please check that:\\r\\n- Source Name isn\\'t already used\\r\\n- The Path is valid\")")
    MessageType=Ok
    DefaultValue=Ok
    MessageCategory=Error
   )
   n9 = "On Item Double Clicked (SourceList)"!() -> (OutputDelegate, then, Item:Object Reference) {is=title K2Node_ComponentBoundEvent}
   > n10 = Cast To SourceEntryItemData!(execute, Object=@n9.Item) -> (then, CastFailed, AsSource Entry Item Data:Source Entry Item Data Object Reference) {is=title K2Node_DynamicCast}
   > Execute Python Script!(
    execute
    PythonScript="import os\r\nimport subprocess\r\nimport unreal\r\n\r\nsrc = unreal.resolve_python_object_handle(src_handle)\r\nsubprocess.Popen(f'explorer \"{os.path.normpath(src.path)}\"')"
    src_handle:Python Object Handle Object Reference=Get item handle!(self:Source Entry Item Data Object Reference=@n10) -> (item handle:Python Object Handle Object Reference)
   ) -> (then, ReturnValue) {inputs [src_handle]}
   "On Clicked (RefreshSourceBtn)"() {is=title K2Node_ComponentBoundEvent}
   > OnSourceChanged(self=Get parent())
   n11 = "On Check State Changed (chk_createImportChannels)"!() -> (OutputDelegate, then, bIsChecked:bool) {is=title K2Node_ComponentBoundEvent}
   > Execute Python Script!(
    execute
    PythonScript="import unreal\r\n\r\nobj = unreal.resolve_python_object_handle(handle)\r\nobj.create_missing_channels = is_checked"
    handle:Python Object Handle Object Reference=@n2
    is_checked:bool=@n11.bIsChecked
   ) -> (then, ReturnValue) {inputs [handle,is_checked]}
   Event OnDialogOpened() {is=title K2Node_Event}
   > InitSettings!(execute, self, parent:EUW Anim Library Object Reference) -> (then)
   Event PreConstruct()
   Event Construct()
   Event Tick()
  ]}
 GetSelectedSource {nodes [
   GetSelectedSource() {is=title K2Node_FunctionEntry}
   > n1 = Cast To SourceEntryItemData(Object=Get Selected Item!(self:List View Object Reference=Get SourceList() -> (SourceList:Editor Utility List View Object Reference)) -> (ReturnValue:Object Reference) {is=title K2Node_CallFunction})
   > Return Node!(execute, AsSource Entry Item Data:Source Entry Item Data Object Reference=@n1)
   Return Node(execute=@n1.CastFailed)
  ]}
 InitSettings {nodes [
   InitSettings!() -> (then, parent:EUW Anim Library Object Reference) {is=title K2Node_FunctionEntry}
   > RefreshOptions()
   > RefreshSources()
  ]}
 RefreshOptions {nodes [
   RefreshOptions() {is=title K2Node_FunctionEntry}
   n1 = Get parent()
   > n2 = Execute Python Script!(
    execute
    PythonScript="import unreal\r\n\r\nobj = unreal.resolve_python_object_handle(handle)\r\nas_binary = obj.as_binary\r\nuse_reducer = obj.use_reducer\r\ncreate_missing_channels = obj.create_missing_channels"
    handle=Get Settings(self=Get library(self=@n1))
   ) -> (then, ReturnValue, as_binary:bool=~, use_reducer:bool=~, create_missing_channels:bool=~) {inputs [handle], outputs [as_binary, use_reducer, create_missing_channels]}
   > SetIsChecked!(
    execute
    self:Check Box Object Reference=Get chk_as_binary() -> (chk_as_binary:Editor Utility Check Box Object Reference)
    InIsChecked:bool=@n2.as_binary
   ) -> (then)
   > SetIsChecked(self=Get chk_use_reducer() -> (chk_use_reducer:Editor Utility Check Box Object Reference), InIsChecked=@n2.use_reducer)
   > SetIsChecked(
    self=Get chk_createImportChannels() -> (chk_createImportChannels:Editor Utility Check Box Object Reference)
    InIsChecked=@n2.create_missing_channels
   )
  ]}
 RefreshSources {nodes [
   RefreshSources() {is=title K2Node_FunctionEntry}
   n1 = Get SourceList()
   > ClearListItems(self=@n1)
   n2 = Get library(self=Get parent())
   > n3 = For Each Loop(Array:Array of Python Object Handle Object References=Get Sources!(self:Anim Lib Manager Object Reference=@n2) -> (Sources:Array of Python Object Handle Object References)) -> (Array Element:Python Object Handle Object Reference)
   > n4 = Construct Object from Class(Class=&/SourceEntryItemData.SourceEntryItemData_C) -> (ReturnValue:Source Entry Item Data Object Reference)
   > Set parentTab!(execute, parentTab:EUW Anim Library Tab Settings Object Reference=Self-Reference(), self:Source Entry Item Data Object Reference=@n4) -> (then, Output_Get:EUW Anim Library Tab Settings Object Reference)
   > Set library!(execute, library:Anim Lib Manager Object Reference=@n2, self:Source Entry Item Data Object Reference=@n4) -> (then, Output_Get:Anim Lib Manager Object Reference)
   > Set item handle!(execute, item handle:Python Object Handle Object Reference=@n3, self:Source Entry Item Data Object Reference=@n4) -> (then, Output_Get:Python Object Handle Object Reference)
   > AddItem!(execute, self:List View Object Reference=@n1, Item:Object Reference=@n4) -> (then)
  ]}
 validateSourceInput {nodes [
   n1 = validateSourceInput!() -> (then, name:String, path:String) {is=title K2Node_FunctionEntry}
   > n2 = GetSourcesData!(execute, self:Anim Lib Manager Object Reference=Get library(self=Get parent())) -> (then, Names:Array of Strings, Paths:Array of Strings)
   n3 = IsEmpty(InString=@n1)
   n4 = IsEmpty(InString=@n1.path)
   n5 = OR Boolean(
    A=Contains Item(TargetArray:Array of Strings=@n2, ItemToFind:"String (by ref)"=@n1)
    B=Contains Item(TargetArray=@n2.Paths, ItemToFind=@n1.path)
   )
   > Return Node!(execute, valid:bool=NOT Boolean(A=OR Boolean!(A=@n5, B=@n3, C:bool=@n4) -> (ReturnValue)))
  ]}
]
variables [
 Library {type=AnimLibManager_C, default=null}
 parent {type=EUW_AnimLibrary_C, default=null}
]
widgets {
 info {parentClass=/Script/Blutility.EditorUtilityDialogWidget, rootWidgetClass=&CanvasPanel, widgetCount=27}
 widgets [
  CanvasPanel_2 {widgetClassPath=&CanvasPanel, widget=CanvasPanel_2}
  VerticalBox_38 {
   widgetClassPath=&VerticalBox
   parent=CanvasPanel_2
   widget=VerticalBox_38
   slot {
    class=CanvasPanelSlot
    properties {bAutoSize=true, layoutData {alignment=<0 0>, anchors {maximum=<1 1>, minimum=<0 0>}, offsets=<ltrb 10 10 10 20>}}
   }
  }
  Options {
   widgetClassPath=&TextBlock
   parent=VerticalBox_38
   widget=Options
   properties {
    text=Options
    colorAndOpacity {colorUseRule=UseColor_Specified, specifiedColor=<rgba 0.807 0.807 0.807 1>}
    font {
     bForceMonospaced=false
     bMaterialIsStencil=false
     fontMaterial=null
     fontObject=&Roboto
     letterSpacing=0
     monospacedWidth=1
     size=18
     skewAmount=0
     typefaceFontName=Bold
     outlineSettings {
      bApplyOutlineToDropShadows=false
      bMiteredCorners=false
      bSeparateFillAlpha=false
      outlineMaterial=null
      outlineSize=0
      outlineColor=<rgba 0 0 0 1>
     }
    }
   }
   slot {class=VerticalBoxSlot, properties {horizontalAlignment=HAlign_Left, verticalAlignment=VAlign_Center}}
  }
  chk_use_reducer {
   widgetClassPath=&EditorUtilityCheckBox
   bIsVariable=true
   parent=VerticalBox_38
   widget=chk_use_reducer
   slot {class=VerticalBoxSlot, properties {padding=<ltrb 4 4 4 4>}}
  }
  UseKeyReducer {
   widgetClassPath=&TextBlock
   parent=chk_use_reducer
   widget=UseKeyReducer
   properties {
    text=Use key reducer
    colorAndOpacity {colorUseRule=UseColor_Specified, specifiedColor=<rgba 0.807 0.807 0.807 1>}
    font {
     bForceMonospaced=false
     bMaterialIsStencil=false
     fontMaterial=null
     fontObject=&Roboto
     letterSpacing=0
     monospacedWidth=1
     size=12
     skewAmount=0
     typefaceFontName=Regular
     outlineSettings {
      bApplyOutlineToDropShadows=false
      bMiteredCorners=false
      bSeparateFillAlpha=false
      outlineMaterial=null
      outlineSize=0
      outlineColor=<rgba 0 0 0 1>
     }
    }
   }
   slot {class=PanelSlot}
  }
  chk_as_binary {
   widgetClassPath=&EditorUtilityCheckBox
   bIsVariable=true
   parent=VerticalBox_38
   widget=chk_as_binary
   slot {class=VerticalBoxSlot, properties {padding=<ltrb 4 4 4 4>}}
  }
  UseKeyReducer_1 {
   widgetClassPath=&TextBlock
   parent=chk_as_binary
   widget=UseKeyReducer_1
   properties {
    text=Use binary format
    colorAndOpacity {colorUseRule=UseColor_Specified, specifiedColor=<rgba 0.807 0.807 0.807 1>}
    font {
     bForceMonospaced=false
     bMaterialIsStencil=false
     fontMaterial=null
     fontObject=&Roboto
     letterSpacing=0
     monospacedWidth=1
     size=12
     skewAmount=0
     typefaceFontName=Regular
     outlineSettings {
      bApplyOutlineToDropShadows=false
      bMiteredCorners=false
      bSeparateFillAlpha=false
      outlineMaterial=null
      outlineSize=0
      outlineColor=<rgba 0 0 0 1>
     }
    }
   }
   slot {class=PanelSlot}
  }
  chk_createImportChannels {
   widgetClassPath=&EditorUtilityCheckBox
   bIsVariable=true
   parent=VerticalBox_38
   widget=chk_createImportChannels
   slot {class=VerticalBoxSlot, properties {padding=<ltrb 4 4 4 4>}}
  }
  Createmissingchannels {
   widgetClassPath=&TextBlock
   parent=chk_createImportChannels
   widget=Createmissingchannels
   properties {
    text=Create missing channels
    toolTipText="If checked, will create missing channels on destination section from the source section, instead of skipping them."
    colorAndOpacity {colorUseRule=UseColor_Specified, specifiedColor=<rgba 0.807 0.807 0.807 1>}
    font {
     bForceMonospaced=false
     bMaterialIsStencil=false
     fontMaterial=null
     fontObject=&Roboto
     letterSpacing=0
     monospacedWidth=1
     size=12
     skewAmount=0
     typefaceFontName=Regular
     outlineSettings {
      bApplyOutlineToDropShadows=false
      bMiteredCorners=false
      bSeparateFillAlpha=false
      outlineMaterial=null
      outlineSize=0
      outlineColor=<rgba 0 0 0 1>
     }
    }
   }
   slot {class=PanelSlot}
  }
  Border_44 {
   widgetClassPath=&Border
   parent=VerticalBox_38
   widget=Border_44
   properties {background {
     bIsDynamicallyLoaded=false
     drawAs=RoundedBox
     imageType=NoImage
     mirroring=NoMirror
     resourceName=None
     resourceObject=null
     tiling=NoTile
     imageSize=<32 32>
     margin=<ltrb 0 0 1 1>
     outlineSettings {
      bUseBrushTransparency=false
      roundingType=FixedRadius
      width=0.5
      color {colorUseRule=UseColor_Specified, specifiedColor=<rgba 1 1 1 1>}
      cornerRadii=<4 4 4 4>
     }
     tintColor {colorUseRule=UseColor_Specified, specifiedColor=<rgba 1 1 1 0>}
     uVRegion {bIsValid=false, max=<0 0>, min=<0 0>}
    }}
   slot {class=VerticalBoxSlot, properties {padding=<ltrb 0 2 0 2>}}
  }
  VerticalBox_49 {widgetClassPath=&VerticalBox, parent=Border_44, widget=VerticalBox_49, slot {class=BorderSlot}}
  HorizontalBox_0 {
   widgetClassPath=&HorizontalBox
   parent=VerticalBox_49
   widget=HorizontalBox_0
   slot {class=VerticalBoxSlot, properties {verticalAlignment=VAlign_Top}}
  }
  Sources {
   widgetClassPath=&TextBlock
   parent=HorizontalBox_0
   widget=Sources
   properties {
    text=Root Locations
    colorAndOpacity {colorUseRule=UseColor_Specified, specifiedColor=<rgba 0.807 0.807 0.807 1>}
    font {
     bForceMonospaced=false
     bMaterialIsStencil=false
     fontMaterial=null
     fontObject=&Roboto
     letterSpacing=0
     monospacedWidth=1
     size=18
     skewAmount=0
     typefaceFontName=Bold
     outlineSettings {
      bApplyOutlineToDropShadows=false
      bMiteredCorners=false
      bSeparateFillAlpha=false
      outlineMaterial=null
      outlineSize=0
      outlineColor=<rgba 0 0 0 1>
     }
    }
   }
   slot {class=HorizontalBoxSlot, properties {padding=<ltrb 4 4 4 4>, size {sizeRule=Fill, value=1}}}
  }
  RefreshSourceBtn {
   widgetClassPath=&EditorUtilityButton
   parent=HorizontalBox_0
   widget=RefreshSourceBtn
   properties {
    bIsEnabled=false
    toolTipText=Refresh all.
    visibility=Collapsed
    widgetStyle {
     clickedSlateSound {resourceObject=null}
     disabled {
      bIsDynamicallyLoaded=false
      drawAs=RoundedBox
      imageType=NoImage
      mirroring=NoMirror
      resourceName=None
      resourceObject=null
      tiling=NoTile
      imageSize=<0 0>
      margin=<ltrb 0 0 0 0>
      outlineSettings {
       bUseBrushTransparency=false
       roundingType=FixedRadius
       width=1
       color {colorUseRule=UseColor_ColorTable, specifiedColor=<rgba 0 0 0 0>}
       cornerRadii=<4 4 4 4>
      }
      tintColor {colorUseRule=UseColor_Specified, specifiedColor=<rgba 0.0395 0.0395 0.0395 1>}
      uVRegion {bIsValid=false, max=<0 0>, min=<0 0>}
     }
     disabledForeground {colorUseRule=UseColor_ColorTable, specifiedColor=<rgba 0 0 0 0>}
     hovered {
      bIsDynamicallyLoaded=false
      drawAs=RoundedBox
      imageType=NoImage
      mirroring=NoMirror
      resourceName=None
      resourceObject=null
      tiling=NoTile
      imageSize=<32 32>
      margin=<ltrb 0 0 0 0>
      outlineSettings {
       bUseBrushTransparency=false
       roundingType=FixedRadius
       width=1
       color {colorUseRule=UseColor_ColorTable, specifiedColor=<rgba 0 0 0 0>}
       cornerRadii=<4 4 4 4>
      }
      tintColor {colorUseRule=UseColor_Specified, specifiedColor=<rgba 0 0.1087 0.5 1>}
      uVRegion {bIsValid=false, max=<0 0>, min=<0 0>}
     }
     hoveredForeground {colorUseRule=UseColor_ColorTable, specifiedColor=<rgba 0 0 0 0>}
     hoveredSlateSound {resourceObject=null}
     normal {
      bIsDynamicallyLoaded=false
      drawAs=NoDrawType
      imageType=FullColor
      mirroring=NoMirror
      resourceName=None
      resourceObject=null
      tiling=NoTile
      imageSize=<24 24>
      margin=<ltrb 0 0 0 0>
      outlineSettings {
       bUseBrushTransparency=false
       roundingType=FixedRadius
       width=1
       color {colorUseRule=UseColor_ColorTable, specifiedColor=<rgba 0 0 0 0>}
       cornerRadii=<4 4 4 4>
      }
      tintColor {colorUseRule=UseColor_Specified, specifiedColor=<rgba 0.807 0.807 0.807 1>}
      uVRegion {bIsValid=false, max=<0 0>, min=<0 0>}
     }
     normalForeground {colorUseRule=UseColor_ColorTable, specifiedColor=<rgba 0 0 0 0>}
     normalPadding=<ltrb 2 2 2 2>
     pressed {
      bIsDynamicallyLoaded=false
      drawAs=RoundedBox
      imageType=NoImage
      mirroring=NoMirror
      resourceName=None
      resourceObject=null
      tiling=NoTile
      imageSize=<32 32>
      margin=<ltrb 0 0 0 0>
      outlineSettings {
       bUseBrushTransparency=false
       roundingType=FixedRadius
       width=1
       color {colorUseRule=UseColor_ColorTable, specifiedColor=<rgba 0 0 0 0>}
       cornerRadii=<4 4 4 4>
      }
      tintColor {colorUseRule=UseColor_Specified, specifiedColor=<rgba 0.0284 0.0284 0.0284 1>}
      uVRegion {bIsValid=false, max=<0 0>, min=<0 0>}
     }
     pressedForeground {colorUseRule=UseColor_ColorTable, specifiedColor=<rgba 0 0 0 0>}
     pressedPadding=<ltrb 2 2 2 2>
     pressedSlateSound {resourceObject=null}
    }
   }
   slot {class=HorizontalBoxSlot, properties {horizontalAlignment=HAlign_Right, padding=<ltrb 2 2 2 2>}}
  }
  btnIconRefresh {
   widgetClassPath=&Image
   bIsVariable=true
   parent=RefreshSourceBtn
   widget=btnIconRefresh
   properties {brush {
     bIsDynamicallyLoaded=false
     drawAs=Image
     imageType=FullColor
     mirroring=NoMirror
     resourceName=None
     resourceObject=/AnimationLibrary/resources/icons/arrowsrotatesolid
     tiling=NoTile
     imageSize=<24 24>
     margin=<ltrb 0 0 0 0>
     outlineSettings {
      bUseBrushTransparency=false
      roundingType=HalfHeightRadius
      width=0
      color {colorUseRule=UseColor_Specified, specifiedColor=<rgba 0 0 0 0>}
      cornerRadii=<0 0 0 0>
     }
     tintColor {colorUseRule=UseColor_Specified, specifiedColor=<rgba 0.807 0.807 0.807 1>}
     uVRegion {bIsValid=false, max=<0 0>, min=<0 0>}
    }}
   slot {class=ButtonSlot, properties {horizontalAlignment=HAlign_Fill, verticalAlignment=VAlign_Fill, padding=<ltrb 2 2 2 2>}}
  }
  Border {
   widgetClassPath=&Border
   parent=VerticalBox_49
   widget=Border
   properties {
    verticalAlignment=VAlign_Center
    background {
     bIsDynamicallyLoaded=false
     drawAs=RoundedBox
     imageType=NoImage
     mirroring=NoMirror
     resourceName=None
     resourceObject=null
     tiling=NoTile
     imageSize=<32 32>
     margin=<ltrb 0 0 1 1>
     outlineSettings {
      bUseBrushTransparency=false
      roundingType=FixedRadius
      width=0.5
      color {colorUseRule=UseColor_Specified, specifiedColor=<rgba 1 1 1 1>}
      cornerRadii=<4 4 4 4>
     }
     tintColor {colorUseRule=UseColor_Specified, specifiedColor=<rgba 1 1 1 0>}
     uVRegion {bIsValid=false, max=<0 0>, min=<0 0>}
    }
   }
   slot {class=VerticalBoxSlot, properties {padding=<ltrb 2 0 2 0>}}
  }
  HorizontalBox_2 {
   widgetClassPath=&HorizontalBox
   parent=Border
   widget=HorizontalBox_2
   properties {bIsEnabled=false, bOverride_Cursor=true, visibility=Collapsed}
   slot {class=BorderSlot, properties {verticalAlignment=VAlign_Center}}
  }
  VerticalBox_52 {
   widgetClassPath=&VerticalBox
   parent=HorizontalBox_2
   widget=VerticalBox_52
   slot {class=HorizontalBoxSlot, properties {verticalAlignment=VAlign_Center, size {sizeRule=Fill, value=1}}}
  }
  HorizontalBox {
   widgetClassPath=&HorizontalBox
   parent=VerticalBox_52
   widget=HorizontalBox
   slot {class=VerticalBoxSlot, properties {verticalAlignment=VAlign_Top, size {sizeRule=Automatic, value=0.95}}}
  }
  Sources_2 {
   widgetClassPath=&TextBlock
   parent=HorizontalBox
   widget=Sources_2
   properties {
    minDesiredWidth=60
    text=Name
    colorAndOpacity {colorUseRule=UseColor_Specified, specifiedColor=<rgba 0.807 0.807 0.807 1>}
    font {
     bForceMonospaced=false
     bMaterialIsStencil=false
     fontMaterial=null
     fontObject=&Roboto
     letterSpacing=0
     monospacedWidth=1
     size=12
     skewAmount=0
     typefaceFontName=Regular
     outlineSettings {
      bApplyOutlineToDropShadows=false
      bMiteredCorners=false
      bSeparateFillAlpha=false
      outlineMaterial=null
      outlineSize=0
      outlineColor=<rgba 0 0 0 1>
     }
    }
   }
   slot {class=HorizontalBoxSlot, properties {horizontalAlignment=HAlign_Left, padding=<ltrb 4 4 4 4>, size {sizeRule=Fill, value=0.17}}}
  }
  edt_NewName {
   widgetClassPath=&EditorUtilityEditableTextBox
   bIsVariable=true
   parent=HorizontalBox
   widget=edt_NewName
   properties {bOverride_Cursor=true, clipping=ClipToBounds, hintText=type New Source Name, minimumDesiredWidth=200}
   slot {
    class=HorizontalBoxSlot
    properties {horizontalAlignment=HAlign_Left, verticalAlignment=VAlign_Center, padding=<ltrb 4 4 4 4>, size {sizeRule=Automatic, value=0.8}}
   }
  }
  HorizontalBox_1 {
   widgetClassPath=&HorizontalBox
   parent=VerticalBox_52
   widget=HorizontalBox_1
   slot {class=VerticalBoxSlot, properties {verticalAlignment=VAlign_Bottom}}
  }
  Sources_3 {
   widgetClassPath=&TextBlock
   parent=HorizontalBox_1
   widget=Sources_3
   properties {
    minDesiredWidth=60
    text=Path
    colorAndOpacity {colorUseRule=UseColor_Specified, specifiedColor=<rgba 0.807 0.807 0.807 1>}
    font {
     bForceMonospaced=false
     bMaterialIsStencil=false
     fontMaterial=null
     fontObject=&Roboto
     letterSpacing=0
     monospacedWidth=1
     size=12
     skewAmount=0
     typefaceFontName=Regular
     outlineSettings {
      bApplyOutlineToDropShadows=false
      bMiteredCorners=false
      bSeparateFillAlpha=false
      outlineMaterial=null
      outlineSize=0
      outlineColor=<rgba 0 0 0 1>
     }
    }
   }
   slot {class=HorizontalBoxSlot, properties {horizontalAlignment=HAlign_Left, padding=<ltrb 4 4 4 4>, size {sizeRule=Fill, value=0.17}}}
  }
  edt_NewPath {
   widgetClassPath=&EditorUtilityEditableTextBox
   bIsVariable=true
   parent=HorizontalBox_1
   widget=edt_NewPath
   properties {bOverride_Cursor=true, clipping=ClipToBounds, hintText=type /New/Path/To/Source, minimumDesiredWidth=200}
   slot {
    class=HorizontalBoxSlot
    properties {horizontalAlignment=HAlign_Left, verticalAlignment=VAlign_Center, padding=<ltrb 4 4 4 4>, size {sizeRule=Automatic, value=0.8}}
   }
  }
  AddSourceBtn {
   widgetClassPath=&EditorUtilityButton
   parent=HorizontalBox_2
   widget=AddSourceBtn
   properties {
    toolTipText=Add a new source
    widgetStyle {
     clickedSlateSound {resourceObject=null}
     disabled {
      bIsDynamicallyLoaded=false
      drawAs=RoundedBox
      imageType=NoImage
      mirroring=NoMirror
      resourceName=None
      resourceObject=null
      tiling=NoTile
      imageSize=<0 0>
      margin=<ltrb 0 0 0 0>
      outlineSettings {
       bUseBrushTransparency=false
       roundingType=FixedRadius
       width=1
       color {colorUseRule=UseColor_ColorTable, specifiedColor=<rgba 0 0 0 0>}
       cornerRadii=<4 4 4 4>
      }
      tintColor {colorUseRule=UseColor_Specified, specifiedColor=<rgba 0.0395 0.0395 0.0395 1>}
      uVRegion {bIsValid=false, max=<0 0>, min=<0 0>}
     }
     disabledForeground {colorUseRule=UseColor_ColorTable, specifiedColor=<rgba 0 0 0 0>}
     hovered {
      bIsDynamicallyLoaded=false
      drawAs=RoundedBox
      imageType=NoImage
      mirroring=NoMirror
      resourceName=None
      resourceObject=null
      tiling=NoTile
      imageSize=<32 32>
      margin=<ltrb 0 0 0 0>
      outlineSettings {
       bUseBrushTransparency=false
       roundingType=FixedRadius
       width=1
       color {colorUseRule=UseColor_ColorTable, specifiedColor=<rgba 0 0 0 0>}
       cornerRadii=<4 4 4 4>
      }
      tintColor {colorUseRule=UseColor_Specified, specifiedColor=<rgba 0 0.1087 0.5 1>}
      uVRegion {bIsValid=false, max=<0 0>, min=<0 0>}
     }
     hoveredForeground {colorUseRule=UseColor_ColorTable, specifiedColor=<rgba 0 0 0 0>}
     hoveredSlateSound {resourceObject=null}
     normal {
      bIsDynamicallyLoaded=false
      drawAs=NoDrawType
      imageType=NoImage
      mirroring=NoMirror
      resourceName=None
      resourceObject=null
      tiling=NoTile
      imageSize=<24 24>
      margin=<ltrb 0 0 0 0>
      outlineSettings {
       bUseBrushTransparency=false
       roundingType=FixedRadius
       width=1
       color {colorUseRule=UseColor_ColorTable, specifiedColor=<rgba 0 0 0 0>}
       cornerRadii=<4 4 4 4>
      }
      tintColor {colorUseRule=UseColor_Specified, specifiedColor=<rgba 0 0.1608 0.7451 1>}
      uVRegion {bIsValid=false, max=<0 0>, min=<0 0>}
     }
     normalForeground {colorUseRule=UseColor_ColorTable, specifiedColor=<rgba 0 0 0 0>}
     normalPadding=<ltrb 2 2 2 2>
     pressed {
      bIsDynamicallyLoaded=false
      drawAs=RoundedBox
      imageType=NoImage
      mirroring=NoMirror
      resourceName=None
      resourceObject=null
      tiling=NoTile
      imageSize=<32 32>
      margin=<ltrb 0 0 0 0>
      outlineSettings {
       bUseBrushTransparency=false
       roundingType=FixedRadius
       width=1
       color {colorUseRule=UseColor_ColorTable, specifiedColor=<rgba 0 0 0 0>}
       cornerRadii=<4 4 4 4>
      }
      tintColor {colorUseRule=UseColor_Specified, specifiedColor=<rgba 0.0284 0.0284 0.0284 1>}
      uVRegion {bIsValid=false, max=<0 0>, min=<0 0>}
     }
     pressedForeground {colorUseRule=UseColor_ColorTable, specifiedColor=<rgba 0 0 0 0>}
     pressedPadding=<ltrb 2 2 2 2>
     pressedSlateSound {resourceObject=null}
    }
   }
   slot {class=HorizontalBoxSlot, properties {horizontalAlignment=HAlign_Center, verticalAlignment=VAlign_Center, padding=<ltrb 2 2 2 2>}}
  }
  btnIconAddSrc {
   widgetClassPath=&Image
   bIsVariable=true
   parent=AddSourceBtn
   widget=btnIconAddSrc
   properties {brush {
     bIsDynamicallyLoaded=false
     drawAs=Image
     imageType=FullColor
     mirroring=NoMirror
     resourceName=None
     resourceObject=/AnimationLibrary/resources/icons/plussolid
     tiling=NoTile
     imageSize=<24 24>
     margin=<ltrb 0 0 0 0>
     outlineSettings {
      bUseBrushTransparency=false
      roundingType=HalfHeightRadius
      width=0
      color {colorUseRule=UseColor_Specified, specifiedColor=<rgba 0 0 0 0>}
      cornerRadii=<0 0 0 0>
     }
     tintColor {colorUseRule=UseColor_Specified, specifiedColor=<rgba 0.807 0.807 0.807 1>}
     uVRegion {bIsValid=false, max=<0 0>, min=<0 0>}
    }}
   slot {class=ButtonSlot, properties {padding=<ltrb 2 2 2 2>}}
  }
  SourceList {
   widgetClassPath=/Script/Blutility.EditorUtilityListView
   bIsVariable=true
   parent=VerticalBox_49
   widget=SourceList
   properties {
    bClearSelectionOnClick=true
    entryWidgetClass=&/EWC_SourceEntry.EWC_SourceEntry_C
    horizontalEntrySpacing=4
    numDesignerPreviewEntries=3
    verticalEntrySpacing=4
   }
   slot {class=VerticalBoxSlot, properties {verticalAlignment=VAlign_Top, padding=<ltrb 4 4 4 4>}}
  }
 ]
}
