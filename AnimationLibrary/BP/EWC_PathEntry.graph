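# Editor utility tree-row widget for one virtual folder in the Animation Library, driven by a PathEntryItemData object.
# Indents by depth, with bindings for chevron, folder or drive icons and inline rename.
# A right-click menu offers rename, sub-folder creation, delete, color and Open in Explorer; closing it writes the color to the Python path object.
# Dropping an AnimLibAsset runs inline Python to add it to the folder, then refreshes the tree.
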EditorUtilityWidgetBlueprint /AnimationLibrary/BP/EWC_PathEntry
graphs [
 EventGraph {nodes [
   n1 = "On Menu Open Changed (ContextMenuAnchor)"!() -> (OutputDelegate, then, bIsOpen:bool) {is=title K2Node_ComponentBoundEvent}
   > n2 = Branch(Condition=@n1.bIsOpen)
   Execute Python Script!(
    execute=@n2.else
    PythonScript="import unreal\r\n\r\nmy_path = unreal.resolve_python_object_handle(handle)\r\nmy_path.color = (new_color.r, new_color.g, new_color.b, new_color.a)\r\n"
    handle:Python Object Handle Object Reference=Get VirtualPathHandle() -> (VirtualPathHandle:Python Object Handle Object Reference)
    new_color:Linear Color Structure=Get Color!(self:Path Entry Item Data Object Reference=Get EntryItemData() -> (EntryItemData:Path Entry Item Data Object Reference)) -> (Color:Linear Color Structure)
   ) -> (then, ReturnValue) {inputs [handle,new_color]}
   n3 = "On Text Committed (edt_name)"!() -> (OutputDelegate, then, Text:"Text (by ref)", CommitMethod:ETextCommit Enum=Default) {is=title K2Node_ComponentBoundEvent}
   > rename!(execute, self:Path Entry Item Data Object Reference=Get EntryItemData(), NewName:String="To String (Text)"(InText=@n3.Text)) -> (then)
   > Set isEditing(isEditing:bool=false) -> (Output_Get:bool)
   n4 = OnRename!() -> (OutputDelegate, then, Context Item:Editor Utility Widget Object Reference, Slot Widget:Editor Utility Widget Object Reference) {is=title K2Node_CustomEvent}
   n5 = Get ContextMenuAnchor() -> (ContextMenuAnchor:Menu Anchor Object Reference)
   > Close(self:Menu Anchor Object Reference=@n5)
   > Set isEditing(isEditing=true)
   n6 = Get EntryItemData()
   n7 = Get edt_name() -> (edt_name:Editor Utility Editable Text Box Object Reference)
   > Set Text(
    Text="To Text (String)"(InString=Get Name!(self:Path Entry Item Data Object Reference=@n6) -> (Name:String))
    self:Text Box Object Reference=@n7
   )
   > SetFocus(self=@n7)
   > SetKeyboardFocus(self:Widget Object Reference=@n7)
   n8 = onCreate!() -> (OutputDelegate, then, Context Item:Editor Utility Widget Object Reference, Slot Widget:Editor Utility Widget Object Reference) {is=title K2Node_CustomEvent}
   > Close(self=@n5)
   > SetExpanded!(execute, self, IsExpanded:bool=false) -> (then)
   > create_folder!(execute, self:Path Entry Item Data Object Reference=Get EntryItemData(), NewFolder:String=NewFolder) -> (then)
   > SetExpanded(IsExpanded=true)
   n9 = onDelete!() -> (OutputDelegate, then, Context Item:Editor Utility Widget Object Reference, Slot Widget:Editor Utility Widget Object Reference) {is=title K2Node_CustomEvent}
   > Close(self=@n5)
   > delete!(execute, self:Path Entry Item Data Object Reference=Get EntryItemData(), withAssets:bool=false) -> (then)
   n10 = onColorCHange!() -> (OutputDelegate, then, Context Item:Editor Utility Widget Object Reference, Slot Widget:Editor Utility Widget Object Reference) {is=title K2Node_CustomEvent}
   > n11 = Cast To EUW_ContextMenu_ColorPicker!(execute, Object=@n10.Context Item) -> (then, CastFailed, AsEUW Context Menu Color Picker:EUW Context Menu Color Picker Object Reference) {is=title K2Node_DynamicCast}
   > n12 = Set Color!(
    execute
    Color:Linear Color Structure=Get color!(self:EUW Context Menu Color Picker Object Reference=@n11) -> (color:Linear Color Structure)
    self:Path Entry Item Data Object Reference=Get EntryItemData()
   ) -> (then, Output_Get:Linear Color Structure)
   > SetBrushTintColor!(
    execute
    self:Image Object Reference=Get folder_img() -> (folder_img:Image Object Reference)
    TintColor:Slate Color Structure="To SlateColor (Color)"!(Value:Linear Color Structure=@n12) -> (ReturnValue:Slate Color Structure)
   ) -> (then)
   n13 = onOpenInExplorer!() -> (OutputDelegate, then, Context Item:Editor Utility Widget Object Reference, Slot Widget:Editor Utility Widget Object Reference) {is=title K2Node_CustomEvent}
   > Close(self=@n5)
   > OpenSourcePath(self:Path Entry Item Data Object Reference=Get EntryItemData())
   n14 = Event PreConstruct()
   n15 = Create Widget!(
    execute:Exec=[@n14.then, @n16.then]
    WorldContextObject:Object Reference
    Class:User Widget Class Reference=/Game/NeverCook/UI/ContextMenu/EUW_ContextMenu.EUW_ContextMenu_C
    OwningPlayer:Player Controller Object Reference
   ) -> (then:Exec, ReturnValue:EUW Context Menu Object Reference)
   > Set ContextMenu(ContextMenu:EUW Context Menu Object Reference=@n15) -> (Output_Get:EUW Context Menu Object Reference)
   Event OnDragLeave!() -> (OutputDelegate, then, PointerEvent:Pointer Event Structure, Operation:Drag Drop Operation Object Reference) {is=title K2Node_Event}
   > OnSelectedDisplay!(execute, self, selectionState:bool=false) -> (then)
   n16 = Event OnInitialized()
   n17 = Event OnListItemObjectSet()
   > Set isEditing(isEditing=false)
   > n18 = Cast To PathEntryItemData!(execute, Object=@n17.ListItemObject) -> (then, CastFailed, AsPath Entry Item Data:Path Entry Item Data Object Reference) {is=title K2Node_DynamicCast}
   > n19 = Set EntryItemData(EntryItemData:Path Entry Item Data Object Reference=@n18) -> (Output_Get:Path Entry Item Data Object Reference)
   > OnSelectedDisplay(selectionState=false)
   > Set VirtualPathHandle(VirtualPathHandle:Python Object Handle Object Reference=Get PurePathObject!(self:Path Entry Item Data Object Reference=@n19) -> (PurePathObject:Python Object Handle Object Reference)) -> (Output_Get:Python Object Handle Object Reference)
   > Get_border_entry_Padding!(execute, self) -> (then, Padding:Margin Structure)
   > n20 = Sequence!(execute) -> (then_0, then_1, then_2:Exec, then_3:Exec, then_4:Exec)
   n21 = Get ContextMenu() -> (ContextMenu:EUW Context Menu Object Reference)
   > Add!(
    execute
    self:EUW Context Menu Object Reference=@n21
    Event:"Delegate (by ref)"=@n4
    Category:String=Edit
    Name:String=Rename
    Icon:Texture 2D Object Reference=/AnimationLibrary/resources/icons/pentosquaresolid
    ItemWidget:User Widget Class Reference=/Game/NeverCook/UI/ContextMenu/EUW_ContextMenu_Button.EUW_ContextMenu_Button_C
    Tooltip:String
   ) -> (then, Context Menu Item:Widget Object Reference, ContainerItem:User Widget Object Reference) {is=title K2Node_CallFunction}
   Add(
    execute=@n20.then_1
    self=@n21
    Event=@n8
    Category=Edit
    Name="Create Sub-Folder(s)"
    Icon=/AnimationLibrary/resources/icons/folderplussolid
    ItemWidget=&EUW_ContextMenu_Button_C
   )
   Add(
    execute=@n20.then_2
    self=@n21
    Event=@n9
    Category=Edit
    Name="Delete (includes hierarchy)"
    Icon=/AnimationLibrary/resources/icons/trashsolid
    ItemWidget=&EUW_ContextMenu_Button_C
   )
   n22 = Add(
    execute=@n20.then_3
    self=@n21
    Event=@n10
    Category=Edit
    Name=Set Color
    ItemWidget=/Game/NeverCook/UI/ContextMenu/EUW_ContextMenu_ColorPicker.EUW_ContextMenu_ColorPicker_C
   )
   > n23 = Cast To EUW_ContextMenu_ColorPicker(Object=@n22)
   > Set color!(execute, color:Linear Color Structure=Get Color(self=Get EntryItemData()), self:EUW Context Menu Color Picker Object Reference=@n23) -> (then, Output_Get:Linear Color Structure)
   Add(execute=@n20.then_4, self=@n21, Event=@n13, Name=Open in Explorer, Icon=/AnimationLibrary/resources/icons/folderopensolid)
   n24 = Event On Item Selection Changed!() -> (OutputDelegate, then, bIsSelected:bool) {is=title K2Node_Event}
   > OnSelectedDisplay(selectionState=@n24.bIsSelected)
   n25 = Event On Item Expansion Changed!() -> (OutputDelegate, then, bIsExpanded:bool) {is=title K2Node_Event}
   > SetExpanded(IsExpanded=@n25.bIsExpanded)
   Event On Entry Accept Drop!() -> (OutputDelegate, then, bSuccess:bool) {is=title K2Node_Event}
  ]}
 Get_border_entry_Padding {nodes [
   Get_border_entry_Padding() {is=title K2Node_FunctionEntry}
   n1 = Get Depth!(self:Path Entry Item Data Object Reference=Get EntryItemData()) -> (Depth:int=0)
   n2 = integer64 + integer64(
    A:Integer64=integer64 * integer64(A:Integer64=10, B:Integer64="To Integer64 (Integer)"!(InInt:int=@n1) -> (ReturnValue:Integer64=0)) -> (ReturnValue:Integer64) {is=title K2Node_PromotableOperator}
    B:Integer64=4
   ) -> (ReturnValue:Integer64) {is=title K2Node_PromotableOperator}
   > n3 = Set Padding!(
    execute
    Padding:Margin Structure=Make Margin(Left="To Float (Integer64)"!(InInt:Integer64=@n2) -> (ReturnValue:double), Top=0, Right=0, Bottom=0)
    self:Border Object Reference=Get border_entry() -> (border_entry:Border Object Reference)
   ) -> (then, Output_Get:Margin Structure)
   > Return Node!(execute, Padding:Margin Structure=@n3)
  ]}
 Get_chevon_img_Brush {nodes [
   Get_chevon_img_Brush() {is=title K2Node_FunctionEntry}
   > n1 = Cast To Texture2D(Object=Select(
     Option 0:Object Reference=/AnimationLibrary/resources/icons/caretrightsolid
     Option 1:Object Reference=/AnimationLibrary/resources/icons/caretdownsolid
     Index=Get isExpanded() -> (isExpanded:bool)
    ) -> (ReturnValue:Object Reference))
   > Return Node!(
    execute
    ReturnValue:Slate Brush Structure=MakeBrushFromTexture!(Texture:Texture 2D Object Reference=@n1, Width:int=8, Height:int=12) -> (ReturnValue:Slate Brush Structure)
   )
  ]}
 Get_chevon_img_Visibility {nodes [
   Get_chevon_img_Visibility() {is=title K2Node_FunctionEntry}
   > n1 = has_children!(execute, self:Path Entry Item Data Object Reference=Get EntryItemData()) -> (then, NewParam:bool)
   > Return Node(ReturnValue:ESlateVisibility Enum=Select(Option 0:ESlateVisibility Enum=Hidden, Option 1:ESlateVisibility Enum=Visible, Index=@n1) -> (ReturnValue:ESlateVisibility Enum))
  ]}
 Get_edt_name_Visibility {nodes [
   Get_edt_name_Visibility() {is=title K2Node_FunctionEntry}
   > Return Node(ReturnValue=Select(Option 0=Collapsed, Option 1=Visible, Index=Get isEditing() -> (isEditing:bool)))
  ]}
 Get_folder_img_Brush {nodes [
   Get_folder_img_Brush() {is=title K2Node_FunctionEntry}
   > n1 = Cast To Texture2D(Object=Select(
     Option 0:Object Reference=/AnimationLibrary/resources/icons/foldersolid
     Option 1:Object Reference=&folderopensolid
     Index=Get isExpanded()
    ) -> (ReturnValue:Object Reference))
   n2 = Get IsSourceRoot!(self:Path Entry Item Data Object Reference=Get EntryItemData()) -> (IsSourceRoot:bool)
   > Return Node(ReturnValue:Slate Brush Structure=MakeBrushFromTexture(
     Texture=Select(Option 0:Texture 2D Object Reference=@n1, Option 1:Texture 2D Object Reference=/AnimationLibrary/resources/icons/harddrivesolid, Index=@n2) -> (ReturnValue:Texture 2D Object Reference)
     Width=20
     Height=20
    ))
  ]}
 Get_item_name_Visibility {nodes [
   Get_item_name_Visibility() {is=title K2Node_FunctionEntry}
   > Return Node(ReturnValue:ESlateVisibility Enum=Select(Option 0:ESlateVisibility Enum=Visible, Option 1:ESlateVisibility Enum=Collapsed, Index=Get isEditing()) -> (ReturnValue:ESlateVisibility Enum))
  ]}
 NewFunction {nodes [NewFunction() {is=title K2Node_FunctionEntry}]}
 OnDragOver {nodes [
   OnDragOver!() -> (then, MyGeometry:Geometry Structure, PointerEvent:Pointer Event Structure, Operation:Drag Drop Operation Object Reference) {is=title K2Node_FunctionEntry}
   > OnSelectedDisplay(selectionState=true)
   > Return Node(ReturnValue:bool=true)
  ]}
 OnDrop {nodes [
   n1 = OnDrop!() -> (then, MyGeometry:Geometry Structure, PointerEvent:Pointer Event Structure, Operation:Drag Drop Operation Object Reference) {is=title K2Node_FunctionEntry}
   > n2 = Cast To AnimLibAsset!(execute, Object=Get Payload!(self:Drag Drop Operation Object Reference=@n1.Operation) -> (Payload:Object Reference)) -> (then, CastFailed, AsAnim Lib Asset:Anim Lib Asset Object Reference) {is=title K2Node_DynamicCast}
   n3 = Get OwnerLibWidget!(self:Path Entry Item Data Object Reference=Get EntryItemData()) -> (OwnerLibWidget:EUW Anim Library Object Reference)
   > n4 = Execute Python Script!(
    execute
    PythonScript="import unreal\r\n\r\nnew_folder = unreal.resolve_python_object_handle(path_handle)\r\nmanager = unreal.resolve_python_object_handle(manager_h)\r\nasset= unreal.resolve_python_object_handle(asset_h)\r\nmanager.folder_manager.add_asset_to_folder(asset, new_folder)\r\n\r\nnew_path = str(new_folder)"
    path_handle:Python Object Handle Object Reference=Get VirtualPathHandle()
    manager_h:Python Object Handle Object Reference=Get Manager!(self:Anim Lib Manager Object Reference=Get library!(self:EUW Anim Library Object Reference=@n3) -> (library:Anim Lib Manager Object Reference)) -> (Manager:Python Object Handle Object Reference)
    asset_h:Python Object Handle Object Reference=Get AssetHandle!(self:Anim Lib Asset Object Reference=@n2) -> (AssetHandle:Python Object Handle Object Reference)
   ) -> (then, ReturnValue, new_path:String) {inputs [path_handle,manager_h,asset_h], outputs [new_path]}
   > PrintString(InString=Append(A="NewPath: ", B=@n4.new_path))
   > OnSelectedDisplay(selectionState=false)
   > RefreshVirtualPathTree(self:EUW Anim Library Object Reference=Get OwnerLibWidget(self=Get EntryItemData()))
   > Return Node(ReturnValue=true)
   edit_path!(execute, self:Anim Lib Asset Object Reference=@n2, new_path:String) -> (then, validated_path:String, is_valid_path:bool)
  ]}
 OnMouseButtonDown {nodes [
   n1 = OnMouseButtonDown!() -> (then, MyGeometry:Geometry Structure, MouseEvent:"Pointer Event Structure (by ref)") {is=title K2Node_FunctionEntry}
   > n2 = Branch(Condition=Is Mouse Button Down!(Input:"Pointer Event Structure (by ref)"=@n1.MouseEvent, MouseButton:Key Structure=RightMouseButton) -> (ReturnValue:bool))
   > ToggleOpen!(execute, self:Menu Anchor Object Reference=Get ContextMenuAnchor(), bFocusOnOpen:bool=true) -> (then)
   n3 = Unhandled()
   > Return Node(ReturnValue:Event Reply Structure=@n3)
   Return Node(execute=@n2.else, ReturnValue=@n3)
  ]}
 OnSelectedDisplay {nodes [
   n1 = OnSelectedDisplay!() -> (then, selectionState:bool=~) {is=title K2Node_FunctionEntry}
   > SetBrushColor(self=Get border_entry(), InBrushColor=SelectColor(A=(R=0.125,G=0.125,B=0.125,A=0.6261), B=(R=0,G=0,B=0,A=0), bPickA=@n1))
  ]}
 On_ContextMenu_GetUserMenuContent {nodes [
   On_ContextMenu_GetUserMenuContent() {is=title K2Node_FunctionEntry}
   n1 = Get ContextMenu()
   > BuildContent(self:EUW Context Menu Object Reference=@n1)
   > Return Node(ReturnValue:User Widget Object Reference=@n1)
  ]}
 SetExpanded {nodes [
   n1 = SetExpanded!() -> (then, IsExpanded:bool=~) {is=title K2Node_FunctionEntry}
   > n2 = has_children(self=Get EntryItemData())
   > Set isExpanded(isExpanded:bool=AND Boolean(A=@n2, B=@n1)) -> (Output_Get:bool)
  ]}
 SetName {nodes [
   n1 = SetName!() -> (then, Name:String) {is=title K2Node_FunctionEntry}
   > Set Text(Text="To Text (String)"(InString=@n1), self:Text Object Reference=Get item_name() -> (item_name:Common Text Object Reference))
   Append(A=@n1)
  ]}
]
variables [
 ContextMenu {type=EUW_ContextMenu_C, default=null}
 EntryItemData {type=PathEntryItemData_C, default=null}
 has_children {type=boolean}
 isEditing {type=boolean}
 isExpanded {type=boolean}
 VirtualPathHandle {type=PythonObjectHandle, default=null}
]
widgets {
 info {rootWidgetClass=&Border, widgetCount=8}
 widgets [
  border_entry {
   widgetClassPath=&Border
   bIsVariable=true
   widget=border_entry
   properties {horizontalAlignment=HAlign_Left, verticalAlignment=VAlign_Top, brushColor=<rgba 1 1 1 0>, padding=<ltrb 4 4 4 4>}
  }
  hb_folder {
   widgetClassPath=&HorizontalBox
   bIsVariable=true
   parent=border_entry
   widget=hb_folder
   properties {bOverride_Cursor=true, visibility=Visible}
   slot {class=BorderSlot, properties {horizontalAlignment=HAlign_Left, verticalAlignment=VAlign_Top, padding=<ltrb 4 4 4 4>}}
  }
  chevon_img {
   widgetClassPath=&Image
   bIsVariable=true
   parent=hb_folder
   widget=chevon_img
   properties {brush {
     bIsDynamicallyLoaded=false
     drawAs=Image
     imageType=FullColor
     mirroring=NoMirror
     resourceName=None
     resourceObject=&caretrightsolid
     tiling=NoTile
     imageSize=<8 12>
     margin=<ltrb 0 0 0 0>
     outlineSettings {
      bUseBrushTransparency=false
      roundingType=HalfHeightRadius
      width=0
      color {colorUseRule=UseColor_Specified, specifiedColor=<rgba 0 0 0 0>}
      cornerRadii=<0 0 0 0>
     }
     tintColor {colorUseRule=UseColor_Specified, specifiedColor=<rgba 0.807 0.807 0.807 1>}
     uVRegion {bIsValid=false, max=<0 0>, min=<0 0>}
    }}
   slot {class=HorizontalBoxSlot, properties {horizontalAlignment=HAlign_Center, verticalAlignment=VAlign_Center}}
  }
  folder_img {
   widgetClassPath=&Image
   bIsVariable=true
   parent=hb_folder
   widget=folder_img
   properties {brush {
     bIsDynamicallyLoaded=false
     drawAs=Image
     imageType=FullColor
     mirroring=NoMirror
     resourceName=None
     resourceObject=&foldersolid
     tiling=NoTile
     imageSize=<20 20>
     margin=<ltrb 0 0 0 0>
     outlineSettings {
      bUseBrushTransparency=false
      roundingType=HalfHeightRadius
      width=0
      color {colorUseRule=UseColor_Specified, specifiedColor=<rgba 0 0 0 0>}
      cornerRadii=<0 0 0 0>
     }
     tintColor {colorUseRule=UseColor_Specified, specifiedColor=<rgba 0.807 0.807 0.807 1>}
     uVRegion {bIsValid=false, max=<0 0>, min=<0 0>}
    }}
   slot {class=HorizontalBoxSlot, properties {padding=<ltrb 4 0 0 0>}}
  }
  Overlay_0 {widgetClassPath=&Overlay, parent=hb_folder, widget=Overlay_0, slot {class=HorizontalBoxSlot, properties {padding=<ltrb 2 0 0 0>}}}
  edt_name {
   widgetClassPath=&EditorUtilityEditableTextBox
   bIsVariable=true
   parent=Overlay_0
   widget=edt_name
   properties {visibility=Collapsed}
   slot {class=OverlaySlot, properties {horizontalAlignment=HAlign_Fill, verticalAlignment=VAlign_Center}}
  }
  item_name {
   widgetClassPath=&CommonTextBlock
   bIsVariable=true
   parent=Overlay_0
   widget=item_name
   properties {
    justification=Center
    text=folder
    visibility=Visible
    wrapTextAt=128
    wrappingPolicy=AllowPerCharacterWrapping
    colorAndOpacity {colorUseRule=UseColor_Specified, specifiedColor=<rgba 0.807 0.807 0.807 1>}
    font {
     bForceMonospaced=false
     bMaterialIsStencil=false
     fontMaterial=null
     fontObject=&Roboto
     letterSpacing=0
     monospacedWidth=1
     size=9
     skewAmount=0
     typefaceFontName=Regular
     outlineSettings {
      bApplyOutlineToDropShadows=false
      bMiteredCorners=false
      bSeparateFillAlpha=false
      outlineMaterial=null
      outlineSize=0
      outlineColor=<rgba 0 0 0 0.5>
     }
    }
    shadowColorAndOpacity=<rgba 0 0 0 0.5>
    shadowOffset=<0.01 0.01>
   }
   slot {class=OverlaySlot, properties {horizontalAlignment=HAlign_Fill, verticalAlignment=VAlign_Center}}
  }
  ContextMenuAnchor {
   widgetClassPath=/Script/UMG.MenuAnchor
   bIsVariable=true
   parent=hb_folder
   widget=ContextMenuAnchor
   properties {menuClass=&EUW_ContextMenu_C, placement=MenuPlacement_BelowRightAnchor}
   slot {class=HorizontalBoxSlot, properties {size {sizeRule=Fill, value=1}}}
  }
 ]
}
