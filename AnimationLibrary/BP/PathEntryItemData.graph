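# Editor-side data object for one virtual folder in the Animation Library widget's path tree.
# On Init reads a Python path handle for name, depth, color and tooltip, then recursively builds child items; source roots show their source path as tooltip.
# Create, rename, move and delete call the library manager's Python folder_manager, with delete also refreshing the tree.
# OpenSourcePath opens a source root's path in Explorer.
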EditorUtilityBlueprint /AnimationLibrary/BP/PathEntryItemData
graphs [
 EventGraph {nodes [
   On Init() {is=title K2Node_CustomEvent}
   n1 = Get OwnerLibWidget() -> (OwnerLibWidget:EUW Anim Library Object Reference)
   > n2 = Execute Python Script!(
    execute
    PythonScript="import unreal\r\nvpath = unreal.resolve_python_object_handle(vpath_handle)\r\nis_src_root = vpath.depth == 1\r\nchildren = []\r\nfor c in sorted(vpath.children):\r\n    children.append(unreal.create_python_object_handle(c))\r\ncolor_tuple = unreal.LinearColor(vpath.color[0], vpath.color[1], vpath.color[2], vpath.color[3])\r\n\r\nfpath = str(vpath)\r\ndepth = vpath.depth\r\nname = vpath.name\r\ntooltip = unreal.Text()\r\nif is_src_root:\r\n    manager = unreal.resolve_python_object_handle(manager)\r\n    src = manager.folder_manager.get_source_for_folder(vpath)\r\n    if src:\r\n        tooltip = unreal.Text(str(src.path))\r\n    else:\r\n        unreal.log_warning(f'src None for vpath: {vpath}')"
    vpath_handle:Python Object Handle Object Reference=Get PurePathObject() -> (PurePathObject:Python Object Handle Object Reference)
    manager:Python Object Handle Object Reference=Get Manager!(self:Anim Lib Manager Object Reference=Get library!(self:EUW Anim Library Object Reference=@n1) -> (library:Anim Lib Manager Object Reference)) -> (Manager:Python Object Handle Object Reference)
   ) -> (
    then
    ReturnValue
    children:Array of Python Object Handle Object References
    is_src_root:bool=~
    color_tuple:Linear Color Structure
    fpath:String
    depth:int
    name:String
    tooltip:Text
   ) {inputs [vpath_handle,manager], outputs [children,is_src_root,color_tuple,fpath,depth,name,tooltip]}
   > Set Depth(Depth:int=@n2.depth) -> (Output_Get:int=0)
   > Set Name(Name:String=@n2.name) -> (Output_Get:String)
   > n3 = Set FullPath(FullPath:String=@n2.fpath) -> (Output_Get:String)
   > n4 = Set IsSourceRoot(IsSourceRoot:bool=@n2.is_src_root) -> (Output_Get:bool)
   > Set Color(Color:Linear Color Structure=@n2.color_tuple) -> (Output_Get:Linear Color Structure)
   > n5 = For Each Loop(Array:Array of Python Object Handle Object References=@n2.children) -> (Array Element:Python Object Handle Object Reference)
   n6 = Branch(execute=@n5.Completed, Condition=@n4)
   > Set ToolTip(ToolTip:Text=@n2.tooltip) -> (Output_Get:Text)
   Set ToolTip(execute=@n6.else, ToolTip="To Text (String)"(InString=@n3))
   create_child!(execute=@n5.LoopBody, self, ChildrenpathHandle:Python Object Handle Object Reference=@n5) -> (then)
  ]}
 OpenSourcePath {nodes [
   OpenSourcePath() {is=title K2Node_FunctionEntry}
   > Branch(Condition=Get IsSourceRoot() -> (IsSourceRoot:bool))
   > Execute Python Script!(
    execute
    PythonScript="import os\r\nimport subprocess\r\n\r\n\r\nsubprocess.Popen(f'explorer \"{os.path.normpath(src_path)}\"')"
    src_path:String="To String (Text)"(InText=Get ToolTip() -> (ToolTip:Text))
   ) -> (then, ReturnValue) {inputs [src_path]}
  ]}
 create_child {nodes [
   n1 = create_child!() -> (then, ChildrenpathHandle:Python Object Handle Object Reference) {is=title K2Node_FunctionEntry}
   > n2 = Construct Object from Class(Class=&/PathEntryItemData.PathEntryItemData_C) -> (ReturnValue:Path Entry Item Data Object Reference)
   > Set PurePathObject!(execute, PurePathObject:Python Object Handle Object Reference=@n1, self:Path Entry Item Data Object Reference=@n2) -> (then, Output_Get:Python Object Handle Object Reference)
   > Set OwnerLibWidget!(execute, OwnerLibWidget:EUW Anim Library Object Reference=Get OwnerLibWidget(), self:Path Entry Item Data Object Reference=@n2) -> (then, Output_Get:EUW Anim Library Object Reference)
   > On Init(self=@n2) {is=title K2Node_CallFunction}
   > Add(
    TargetArray:Array of Path Entry Item Data Object References=Get children() -> (children:Array of Path Entry Item Data Object References)
    NewItem:"Path Entry Item Data Object Reference (by ref)"=@n2
   ) {is=title K2Node_CallArrayFunction}
  ]}
 create_folder {nodes [
   n1 = create_folder!() -> (then, NewFolder:String=NewFolder) {is=title K2Node_FunctionEntry}
   n2 = Get OwnerLibWidget()
   > n3 = Execute Python Script!(
    execute
    PythonScript="import unreal\r\n\r\nvpath = unreal.resolve_python_object_handle(vpath_handle)\r\nmanager = unreal.resolve_python_object_handle(manager)\r\nnew_path = manager.folder_manager.create_folder(name=new_folder, parent_path=vpath.path.as_posix())\r\nnew_path = unreal.create_python_object_handle(new_path)"
    vpath_handle:Python Object Handle Object Reference=Get PurePathObject()
    new_folder:String=@n1
    manager:Python Object Handle Object Reference=Get Manager(self=Get library(self=@n2))
   ) -> (then, ReturnValue, new_path:Python Object Handle Object Reference) {inputs [vpath_handle,new_folder,manager], outputs [new_path]}
   > create_child(ChildrenpathHandle=@n3.new_path) {is=title K2Node_CallFunction}
  ]}
 delete {nodes [
   n1 = delete!() -> (then, withAssets:bool=~) {is=title K2Node_FunctionEntry}
   n2 = Get OwnerLibWidget()
   > Execute Python Script!(
    execute
    PythonScript="import unreal\r\n\r\nvpath = unreal.resolve_python_object_handle(vpath_handle)\r\nmanager = unreal.resolve_python_object_handle(manager)\r\nmanager.folder_manager.delete_folder(vpath, force=force)\r\n"
    vpath_handle=Get PurePathObject()
    force:bool=@n1
    manager=Get Manager(self=Get library(self=@n2))
   ) -> (then, ReturnValue) {inputs [vpath_handle,force,manager]}
   > RefreshVirtualPathTree(self:EUW Anim Library Object Reference=@n2)
  ]}
 has_children {nodes [
   has_children() {is=title K2Node_FunctionEntry}
   > Return Node!(execute, NewParam:bool=Is Not Empty!(TargetArray:Array of Path Entry Item Data Object References=Get children()) -> (ReturnValue:bool))
  ]}
 move {nodes [
   n1 = move!() -> (then, NewParent:Path Entry Item Data Object Reference) {is=title K2Node_FunctionEntry}
   n2 = Get OwnerLibWidget()
   > Execute Python Script!(
    execute
    PythonScript="import unreal\r\n\r\nvpath = unreal.resolve_python_object_handle(vpath_handle)\r\ndestination = unreal.resolve_python_object_handle(destination)\r\n\r\nmanager = unreal.resolve_python_object_handle(manager)\r\n\r\nmanager.folder_manager.move_folder(vpath, destination)\r\n"
    vpath_handle=Get PurePathObject()
    destination:Python Object Handle Object Reference=Get PurePathObject!(self:Path Entry Item Data Object Reference=@n1) -> (PurePathObject)
    manager=Get Manager(self=Get library(self=@n2))
   ) -> (then, ReturnValue) {inputs [vpath_handle,destination,manager]}
  ]}
 rename {nodes [
   n1 = rename!() -> (then, NewName:String) {is=title K2Node_FunctionEntry}
   > n2 = Set Name(Name=@n1)
   n3 = Get OwnerLibWidget()
   > Execute Python Script!(
    execute
    PythonScript="import unreal\r\n\r\nvpath = unreal.resolve_python_object_handle(vpath_handle)\r\nmanager = unreal.resolve_python_object_handle(manager)\r\nmanager.folder_manager.rename_folder(vpath, new_name)\r\n"
    vpath_handle=Get PurePathObject!() -> (PurePathObject)
    new_name:String=@n2
    manager=Get Manager(self=Get library(self=@n3))
   ) -> (then, ReturnValue) {inputs [vpath_handle,new_name,manager]}
  ]}
]
variables [
 children {type="array<PathEntryItemData_C>", default []}
 Color {type=LinearColor, default=<rgba 0.807 0.807 0.807 1>}
 Depth {type=integer}
 FullPath {type=string, default=""}
 IsSourceRoot {type=boolean}
 Name {type=string, default=""}
 OwnerLibWidget {type=EUW_AnimLibrary_C, default=null}
 PurePathObject {type=PythonObjectHandle, default=null}
 ToolTip {type=string, default=""}
]
