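# Editor utility dialog widget for entering the name of a new selection set in the Animation Library tool.
# It derives from EditorUtilityDialogWidget, with a vertical box holding one editable text field, edtSetName, hinted "Selection Set Name".
# The event graph has only unconnected default events and no variables, so the widget itself contains no logic.
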EditorUtilityWidgetBlueprint /AnimationLibrary/BP/dlgNewSelectionSet
graphs [EventGraph {nodes [
   Event PreConstruct()
   Event Construct()
   Event Tick()
  ]}]
variables []
widgets {
 info {parentClass=/Script/Blutility.EditorUtilityDialogWidget, rootWidgetClass=&VerticalBox, widgetCount=2}
 widgets [
  VerticalBox_0 {widgetClassPath=&VerticalBox, widget=VerticalBox_0, properties {bOverride_Cursor=true}}
  edtSetName {
   widgetClassPath=&EditorUtilityEditableTextBox
   bIsVariable=true
   parent=VerticalBox_0
   widget=edtSetName
   properties {hintText=Selection Set Name}
   slot {class=VerticalBoxSlot}
  }
 ]
}
