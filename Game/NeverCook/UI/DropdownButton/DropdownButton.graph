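# Editor utility widget for a reusable split button: a main action button joined to a caret dropdown button.
# The main button broadcasts MainButtonClicked, while the caret toggles a MenuAnchor below the row.
# On initialization it applies the Font variable to the label and creates one EUW_ContextMenu, which the anchor's content function rebuilds through BuildContent and returns.
# Helper functions supply the MainButtonText label and recursively find the top parent widget.
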EditorUtilityWidgetBlueprint /Game/NeverCook/UI/DropdownButton/DropdownButton
graphs [
 EventGraph {nodes [
   "On Clicked (btnMain)"() {is=title K2Node_ComponentBoundEvent}
   > Call Main Button Clicked!(execute:Exec, self:Dropdown Button Object Reference) -> (then:Exec) {is=title K2Node_CallDelegate}
   "On Clicked (btnDrop)"() {is=title K2Node_ComponentBoundEvent}
   > ToggleOpen!(execute, self:Menu Anchor Object Reference=Get MenuAnchor() -> (MenuAnchor:Menu Anchor Object Reference), bFocusOnOpen:bool=true) -> (then)
   Event OnInitialized()
   > Set Font!(
    execute
    Font:Slate Font Info Structure=Get Font() -> (Font:Slate Font Info Structure)
    self:Text Object Reference=Get btnMainText() -> (btnMainText:Text Object Reference)
   ) -> (then, Output_Get:Slate Font Info Structure)
   > n1 = Create Widget!(
    execute:Exec
    WorldContextObject:Object Reference
    Class:User Widget Class Reference=/Game/NeverCook/UI/ContextMenu/EUW_ContextMenu.EUW_ContextMenu_C
    OwningPlayer:Player Controller Object Reference
   ) -> (then:Exec, ReturnValue:EUW Context Menu Object Reference)
   > Set Menu(Menu:EUW Context Menu Object Reference=@n1) -> (Output_Get:EUW Context Menu Object Reference)
   Event Construct()
  ]}
 GetTopParent {nodes [
   n1 = GetTopParent!() -> (then, Widget:Widget Object Reference) {is=title K2Node_FunctionEntry}
   n2 = GetParent!(self:Widget Object Reference=@n1) -> (ReturnValue:Panel Widget Object Reference)
   > n3 = Is Valid(InputObject=@n2)
   Return Node!(execute=@n3.Is Not Valid, TopParent:Widget Object Reference)
   > n4 = GetTopParent!(execute, self, Widget:Widget Object Reference=@n2) -> (then, TopParent:Widget Object Reference) {is=title K2Node_CallFunction}
   > Return Node(TopParent=@n4)
  ]}
 Get_btnMainText_Text {nodes [
   Get_btnMainText_Text() {is=title K2Node_FunctionEntry}
   > Return Node!(execute, ReturnValue:Text=Get MainButtonText() -> (MainButtonText:Text))
  ]}
 MainButtonClicked {nodes [MainButtonClicked() {is=title K2Node_FunctionEntry}]}
 On_MenuAnchor_GetUserMenuContent {nodes [
   On_MenuAnchor_GetUserMenuContent() {is=title K2Node_FunctionEntry}
   n1 = Get Menu() -> (Menu:EUW Context Menu Object Reference)
   > BuildContent(self:EUW Context Menu Object Reference=@n1)
   > Return Node(ReturnValue:User Widget Object Reference=@n1)
  ]}
]
variables [
 Font {
  type=SlateFontInfo
  default {
   bForceMonospaced=false
   bMaterialIsStencil=false
   fontMaterial=null
   fontObject=&Roboto
   letterSpacing=0
   monospacedWidth=1
   size=18
   skewAmount=0
   typefaceFontName=Regular
   outlineSettings {
    bApplyOutlineToDropShadows=false
    bMiteredCorners=false
    bSeparateFillAlpha=false
    outlineMaterial=null
    outlineSize=0
    outlineColor=<rgba 0 0 0 1>
   }
  }
 }
 MainButtonClicked {default=()}
 MainButtonText {type=string, default=Main}
 Menu {type=EUW_ContextMenu_C, default=null}
]
widgets {
 info {rootWidgetClass=&VerticalBox, widgetCount=7}
 widgets [
  OverlayDropMenu {widgetClassPath=&VerticalBox, bIsVariable=true, widget=OverlayDropMenu, properties {bOverride_Cursor=true, visibility=Visible}}
  buttonBox {
   widgetClassPath=&HorizontalBox
   bIsVariable=true
   parent=OverlayDropMenu
   widget=buttonBox
   properties {bOverride_Cursor=true}
   slot {class=VerticalBoxSlot}
  }
  btnMain {
   widgetClassPath=&EditorUtilityButton
   parent=buttonBox
   widget=btnMain
   properties {widgetStyle {
     clickedSlateSound {resourceObject=null}
     disabled {
      bIsDynamicallyLoaded=false
      drawAs=RoundedBox
      imageType=NoImage
      mirroring=NoMirror
      resourceName=None
      resourceObject=null
      tiling=NoTile
      imageSize=<0 0>
      margin=<ltrb 0 0 0 0>
      outlineSettings {
       bUseBrushTransparency=false
       roundingType=FixedRadius
       width=0
       color {colorUseRule=UseColor_Foreground, specifiedColor=<rgba 0 0 0 0>}
       cornerRadii=<6 0 0 6>
      }
      tintColor {colorUseRule=UseColor_Foreground, specifiedColor=<rgba 0.0395 0.0395 0.0395 1>}
      uVRegion {bIsValid=false, max=<0 0>, min=<0 0>}
     }
     disabledForeground {colorUseRule=UseColor_ColorTable, specifiedColor=<rgba 0 0 0 0>}
     hovered {
      bIsDynamicallyLoaded=false
      drawAs=RoundedBox
      imageType=NoImage
      mirroring=NoMirror
      resourceName=None
      resourceObject=null
      tiling=NoTile
      imageSize=<32 32>
      margin=<ltrb 0 0 0 0>
      outlineSettings {
       bUseBrushTransparency=false
       roundingType=FixedRadius
       width=0
       color {colorUseRule=UseColor_Foreground, specifiedColor=<rgba 0 0 0 0>}
       cornerRadii=<6 0 0 6>
      }
      tintColor {colorUseRule=UseColor_Specified, specifiedColor=<rgba 0.0953 0.0953 0.0953 1>}
      uVRegion {bIsValid=false, max=<0 0>, min=<0 0>}
     }
     hoveredForeground {colorUseRule=UseColor_ColorTable, specifiedColor=<rgba 0 0 0 0>}
     hoveredSlateSound {resourceObject=null}
     normal {
      bIsDynamicallyLoaded=false
      drawAs=RoundedBox
      imageType=NoImage
      mirroring=NoMirror
      resourceName=None
      resourceObject=null
      tiling=NoTile
      imageSize=<32 32>
      margin=<ltrb 0 0 0 0>
      outlineSettings {
       bUseBrushTransparency=false
       roundingType=FixedRadius
       width=0
       color {colorUseRule=UseColor_Foreground, specifiedColor=<rgba 0 0 0 0>}
       cornerRadii=<6 0 0 6>
      }
      tintColor {colorUseRule=UseColor_Foreground, specifiedColor=<rgba 0.0395 0.0395 0.0395 1>}
      uVRegion {bIsValid=false, max=<0 0>, min=<0 0>}
     }
     normalForeground {colorUseRule=UseColor_ColorTable, specifiedColor=<rgba 0 0 0 0>}
     normalPadding=<ltrb 12 1.5 12 1.5>
     pressed {
      bIsDynamicallyLoaded=false
      drawAs=RoundedBox
      imageType=NoImage
      mirroring=NoMirror
      resourceName=None
      resourceObject=null
      tiling=NoTile
      imageSize=<32 32>
      margin=<ltrb 0 0 0 0>
      outlineSettings {
       bUseBrushTransparency=false
       roundingType=FixedRadius
       width=0
       color {colorUseRule=UseColor_Foreground, specifiedColor=<rgba 0 0 0 0>}
       cornerRadii=<6 0 0 6>
      }
      tintColor {colorUseRule=UseColor_Foreground, specifiedColor=<rgba 0.0284 0.0284 0.0284 1>}
      uVRegion {bIsValid=false, max=<0 0>, min=<0 0>}
     }
     pressedForeground {colorUseRule=UseColor_ColorTable, specifiedColor=<rgba 0 0 0 0>}
     pressedPadding=<ltrb 12 2.5 12 0.5>
     pressedSlateSound {resourceObject=null}
    }}
   slot {class=HorizontalBoxSlot, properties {horizontalAlignment=HAlign_Left}}
  }
  btnMainText {
   widgetClassPath=&TextBlock
   bIsVariable=true
   parent=btnMain
   widget=btnMainText
   properties {
    clipping=ClipToBounds
    justification=Center
    text=MainAction
    font {
     bForceMonospaced=false
     bMaterialIsStencil=false
     fontMaterial=null
     fontObject=&Roboto
     letterSpacing=0
     monospacedWidth=1
     size=18
     skewAmount=0
     typefaceFontName=Bold
     outlineSettings {
      bApplyOutlineToDropShadows=false
      bMiteredCorners=false
      bSeparateFillAlpha=false
      outlineMaterial=null
      outlineSize=0
      outlineColor=<rgba 0 0 0 1>
     }
    }
   }
   slot {class=ButtonSlot, properties {padding=<ltrb 4 4 4 4>}}
  }
  btnDrop {
   widgetClassPath=&EditorUtilityButton
   parent=buttonBox
   widget=btnDrop
   properties {widgetStyle {
     clickedSlateSound {resourceObject=null}
     disabled {
      bIsDynamicallyLoaded=false
      drawAs=RoundedBox
      imageType=NoImage
      mirroring=NoMirror
      resourceName=None
      resourceObject=null
      tiling=NoTile
      imageSize=<0 0>
      margin=<ltrb 0 0 0 0>
      outlineSettings {
       bUseBrushTransparency=false
       roundingType=FixedRadius
       width=0
       color {colorUseRule=UseColor_ColorTable, specifiedColor=<rgba 0 0 0 0>}
       cornerRadii=<0 6 6 0>
      }
      tintColor {colorUseRule=UseColor_Specified, specifiedColor=<rgba 0.0395 0.0395 0.0395 1>}
      uVRegion {bIsValid=false, max=<0 0>, min=<0 0>}
     }
     disabledForeground {colorUseRule=UseColor_ColorTable, specifiedColor=<rgba 0 0 0 0>}
     hovered {
      bIsDynamicallyLoaded=false
      drawAs=RoundedBox
      imageType=NoImage
      mirroring=NoMirror
      resourceName=None
      resourceObject=null
      tiling=NoTile
      imageSize=<32 32>
      margin=<ltrb 0 0 0 0>
      outlineSettings {
       bUseBrushTransparency=false
       roundingType=FixedRadius
       width=0
       color {colorUseRule=UseColor_ColorTable, specifiedColor=<rgba 0 0 0 0>}
       cornerRadii=<0 6 6 0>
      }
      tintColor {colorUseRule=UseColor_Specified, specifiedColor=<rgba 0.0953 0.0953 0.0953 1>}
      uVRegion {bIsValid=false, max=<0 0>, min=<0 0>}
     }
     hoveredForeground {colorUseRule=UseColor_ColorTable, specifiedColor=<rgba 0 0 0 0>}
     hoveredSlateSound {resourceObject=null}
     normal {
      bIsDynamicallyLoaded=false
      drawAs=RoundedBox
      imageType=FullColor
      mirroring=NoMirror
      resourceName=None
      resourceObject=null
      tiling=NoTile
      imageSize=<94 150>
      margin=<ltrb 0 0 0 0>
      outlineSettings {
       bUseBrushTransparency=false
       roundingType=FixedRadius
       width=0
       color {colorUseRule=UseColor_ColorTable, specifiedColor=<rgba 0 0 0 0>}
       cornerRadii=<0 6 6 0>
      }
      tintColor {colorUseRule=UseColor_Foreground, specifiedColor=<rgba 0.0395 0.0395 0.0395 1>}
      uVRegion {bIsValid=false, max=<0 0>, min=<0 0>}
     }
     normalForeground {colorUseRule=UseColor_ColorTable, specifiedColor=<rgba 0 0 0 0>}
     normalPadding=<ltrb 6 1.5 6 1.5>
     pressed {
      bIsDynamicallyLoaded=false
      drawAs=RoundedBox
      imageType=NoImage
      mirroring=NoMirror
      resourceName=None
      resourceObject=null
      tiling=NoTile
      imageSize=<32 32>
      margin=<ltrb 0 0 0 0>
      outlineSettings {
       bUseBrushTransparency=false
       roundingType=FixedRadius
       width=0
       color {colorUseRule=UseColor_ColorTable, specifiedColor=<rgba 0 0 0 0>}
       cornerRadii=<0 6 6 0>
      }
      tintColor {colorUseRule=UseColor_Specified, specifiedColor=<rgba 0.0284 0.0284 0.0284 1>}
      uVRegion {bIsValid=false, max=<0 0>, min=<0 0>}
     }
     pressedForeground {colorUseRule=UseColor_ColorTable, specifiedColor=<rgba 0 0 0 0>}
     pressedPadding=<ltrb 6 2.5 6 0.5>
     pressedSlateSound {resourceObject=null}
    }}
   slot {class=HorizontalBoxSlot, properties {horizontalAlignment=HAlign_Right}}
  }
  btnDropIcon {
   widgetClassPath=&Image
   bIsVariable=true
   parent=btnDrop
   widget=btnDropIcon
   properties {brush {
     bIsDynamicallyLoaded=false
     drawAs=Image
     imageType=FullColor
     mirroring=NoMirror
     resourceName=None
     resourceObject=/AnimationLibrary/resources/icons/caretdownsolid
     tiling=NoTile
     imageSize=<16 16>
     margin=<ltrb 0 0 0 0>
     outlineSettings {
      bUseBrushTransparency=false
      roundingType=HalfHeightRadius
      width=0
      color {colorUseRule=UseColor_Specified, specifiedColor=<rgba 0 0 0 0>}
      cornerRadii=<0 0 0 0>
     }
     tintColor {colorUseRule=UseColor_Specified, specifiedColor=<rgba 1 1 1 1>}
     uVRegion {bIsValid=false, max=<0 0>, min=<0 0>}
    }}
   slot {class=ButtonSlot, properties {padding=<ltrb 0 4 0 4>}}
  }
  MenuAnchor {
   widgetClassPath=/Script/UMG.MenuAnchor
   bIsVariable=true
   parent=OverlayDropMenu
   widget=MenuAnchor
   properties {menuClass=&EUW_ContextMenu_C}
   slot {class=VerticalBoxSlot}
  }
 ]
}
