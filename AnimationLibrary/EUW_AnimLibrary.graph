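# Editor utility widget that serves as the Animation Library panel for browsing, saving and applying rig animations and poses in Sequencer.
# An Anim Lib Manager supplies assets to a path tree and tile/list views, filtered by search text, type and rig.
# Saving, applying and startup checks call the Python backend through inline scripts, disabling the panel without the unpipe module.
# Applying can create a missing Control Rig track, and selection sets store rig controls.
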EditorUtilityWidgetBlueprint /AnimationLibrary/EUW_AnimLibrary
graphs [
 CaptureThumbnail {nodes [
   CaptureThumbnail() {is=title K2Node_FunctionEntry}
   > update_thumbnail!(execute, self:Anim Lib Asset Object Reference=Get Selected_Asset() -> (Selected_Asset:Anim Lib Asset Object Reference)) -> (then, SequencePath:String)
  ]}
 CheckUnpipe {nodes [
   CheckUnpipe() {is=title K2Node_FunctionEntry}
   > n1 = Execute Python Script!(execute, PythonScript="valid_unpipe = True\r\ntry:\r\n    import unpipe\r\nexcept:\r\n   valid_unpipe = False") -> (then, ReturnValue, valid_unpipe:bool=~) {outputs [valid_unpipe]}
   > Return Node!(execute, ValidUnpipe:bool=AND Boolean(A=@n1, B=@n1.valid_unpipe))
  ]}
 ClearSelectionFull {nodes [
   ClearSelectionFull() {is=title K2Node_FunctionEntry}
   > n1 = Sequence!(execute) -> (then_0, then_1, then_2:Exec)
   n2 = Get Assets() -> (Assets:Tile View Object Reference)
   > Clear Selection(self:List View Object Reference=@n2)
   > RegenerateAllEntries(self:List View Base Object Reference=@n2)
   > Clear Selection(self=@n2)
   > Set Selected Item!(execute, self:List View Object Reference=@n2, Item:Object Reference) -> (then) {is=title K2Node_CallFunction}
   n3 = Get AssetsAsList() -> (AssetsAsList:Editor Utility List View Object Reference)
   Clear Selection(execute=@n1.then_1, self=@n3)
   > RegenerateAllEntries(self=@n3)
   > Clear Selection(self=@n3)
   > Set Selected Item(self=@n3)
   > Return Node!(execute, Assets:Tile View Object Reference=@n2)
   Set Selected_Asset(execute=@n1.then_2, Selected_Asset:Anim Lib Asset Object Reference) -> (Output_Get:Anim Lib Asset Object Reference)
  ]}
 CreateSelectionSet {nodes [
   n1 = CreateSelectionSet!() -> (then, name:String) {is=title K2Node_FunctionEntry}
   > n2 = GetCurrentAssetRigInSequencer!(execute, self) -> (then, Output_Rig:Control Rig Object Reference)
   > n3 = AddSelectionSet!(
    execute
    self:Anim Lib Manager Object Reference=Get library() -> (library:Anim Lib Manager Object Reference)
    Name:String=@n1
    Controls:Array of Names=CurrentControlSelection!(self:Control Rig Object Reference=@n2) -> (ReturnValue:Array of Names)
   ) -> (then, new_set:Controls Selection Set Object Reference)
   > PopulateSets()
   > Return Node!(execute, new_set:Controls Selection Set Object Reference=@n3)
  ]}
 EventGraph {nodes [
   n1 = CheckUnpipe!(execute, self) -> (then, ValidUnpipe:bool) {is=title K2Node_CallFunction}
   > n2 = Branch(Condition=@n1)
   SetLibrary(execute=[@n119.then, @n2.then])
   > n3 = refreshAssets!(
    execute
    self
    ResetTree:bool=true
    CalledBy:String=OnInit
    SendNotification:bool=true
    Reconciliations:Map of Strings to Strings=Get Reconciliations() -> (Reconciliations:Map of Strings to Strings)
   ) -> (then)
   n4 = OnRefreshAssets!() -> (OutputDelegate, then, Reconciliations:Map of Strings to Strings) {is=title K2Node_CustomEvent}
   Bind Event to On Refresh Assets(execute=@n3.then, self:Anim Lib Manager Object Reference=Get library(), Delegate=@n4) {is=title K2Node_AddDelegate}
   > PopulateSets()
   > n5 = GetTileSize!(execute, self:Anim Lib Manager Object Reference=Get library()) -> (then, Tile Size:double)
   > Set Value!(execute, Value:float=@n5, self:Slider Object Reference=Get sld_ThumbSizer() -> (sld_ThumbSizer:Editor Utility Slider Object Reference)) -> (then, Output_Get:float)
   > SetObject(self=Get SearchParamsView() -> (SearchParamsView:Details View Object Reference), NewObject=Self-Reference())
   > n6 = Sequence!(execute) -> (then_0, then_1, then_2, then_3:Exec, then_4:Exec, then_5:Exec)
   n7 = Get SaveActions() -> (SaveActions:Dropdown Button Object Reference)
   > Set ToolTipText!(
    execute
    ToolTipText:Text=MakeLiteralText(Value="LOCTEXT(\"Creates a new asset (Animation or Pose).\")")
    self:Widget Object Reference=Get btnMain!(self:Dropdown Button Object Reference=@n7) -> (btnMain:Editor Utility Button Object Reference)
   ) -> (then, Output_Get:Text)
   > Add!(
    execute
    self:EUW Context Menu Object Reference=Get Menu!(self:Dropdown Button Object Reference=@n7) -> (Menu:EUW Context Menu Object Reference)
    Event:"Delegate (by ref)"
    Category:String=Save
    Name:String=copyRange
    Icon:Texture 2D Object Reference=/AnimationLibrary/resources/icons/sliderssolid
    ItemWidget:User Widget Class Reference=/Game/NeverCook/UI/ContextMenu/EUW_ContextMenu_FrameRange.EUW_ContextMenu_FrameRange_C
    Tooltip:String="Save the animation from the specified range (instead of selection based)"
   ) -> (then, Context Menu Item:Widget Object Reference, ContainerItem:User Widget Object Reference) {is=title K2Node_CallFunction}
   n8 = Get PasteActions() -> (PasteActions:Dropdown Button Object Reference)
   Set ToolTipText(
    execute=@n6.then_1
    ToolTipText=MakeLiteralText(Value="LOCTEXT(\"Paste to current selection at current time or original range (depends on options). Based on selected controls in the sequencer\")")
    self=Get btnMain(self=@n8)
   )
   n9 = Get Menu(self=@n8)
   > Add(
    self=@n9
    Category=Apply
    Name=Selected Controls
    ItemWidget=/Game/NeverCook/UI/ContextMenu/EUW_ContextMenu_CheckBox.EUW_ContextMenu_CheckBox_C
   )
   n10 = Add(
    execute=@n6.then_2
    self=@n9
    Category=Apply
    Name=Original FrameRange
    ItemWidget=&EUW_ContextMenu_CheckBox_C
    Tooltip="If checked, Applying frames will always happen at the original times of capture. It'll also apply all saved frames."
   )
   n11 = Add(
    execute=@n6.then_3
    self=@n9
    Category=Apply
    Name=Paste FrameRange
    Icon=&sliderssolid
    ItemWidget=&EUW_ContextMenu_FrameRange_C
    Tooltip=Uses this frame range from the original capture to apply
   )
   > n12 = Cast To EUW_ContextMenu_FrameRange!(execute, Object=@n11) -> (then, CastFailed, AsEUW Context Menu Frame Range:EUW Context Menu Frame Range Object Reference) {is=title K2Node_DynamicCast}
   > Set PasteFrameRange(PasteFrameRange:EUW Context Menu Frame Range Object Reference=@n12) -> (Output_Get:EUW Context Menu Frame Range Object Reference)
   > n13 = Cast To EUW_ContextMenu_CheckBox!(execute, Object=@n10) -> (then, CastFailed, AsEUW Context Menu Check Box:EUW Context Menu Check Box Object Reference) {is=title K2Node_DynamicCast}
   n14 = OnOriginalControlsChecked!() -> (OutputDelegate, then, bIsChecked:bool=~) {is=title K2Node_CustomEvent}
   Bind Event to On Check State Changed(
    execute=@n13.then
    self:Check Box Object Reference=Get Checkbox!(self:EUW Context Menu Check Box Object Reference=@n13) -> (Checkbox:Check Box Object Reference)
    Delegate=@n14
   ) {is=title K2Node_AddDelegate}
   Show Message Dialog(
    execute=@n2.else
    Title=MakeLiteralText(Value="LOCTEXT(\"Animation Library Error\")")
    Message=MakeLiteralText(Value="LOCTEXT(\"The unpipe module couldn\\'t be found. This is required for the library to work.\\r\\nAre you sure you\\'ve loaded your pipeline? \")")
    MessageType=Ok
    DefaultValue=Ok
    MessageCategory=Error
   )
   > SetIsEnabled(self=Self-Reference(), bInIsEnabled=false)
   > n15 = MarkNotificationAsError!(
    execute
    NotificationSettings:"Editor Notification Settings Structure (by ref)"=MakeEditorNotificationSettings!(
     Text:Text="LOCTEXT(\"Animation Library\")"
     SubText:Text="LOCTEXT(\"Pipeline not loaded!\")"
     FadeInDuration:float=0.5
     FadeOutDuration:float=2
     ExpireDuration:float=5
     bUseThrobber:bool=true
     bUseSuccessFailIcons:bool=true
     bFireAndForget:bool=true
     bAllowThrottleWhenFrameRateIsLow:bool=true
    ) -> (ReturnValue:Editor Notification Settings Structure)
   ) -> (then, ReturnValue:Editor Notification Settings Structure)
   > AddEditorNotification!(execute, EditorNotificationSettings:"Editor Notification Settings Structure (by ref)"=@n15) -> (then, ReturnValue:Editor Notification Handle Structure)
   n16 = Get treeview_path() -> (treeview_path:Common Tree View Object Reference)
   n17 = For Each Loop(Array:Array of Object References=GetDisplayedEntryWidgets!(self:List View Base Object Reference=@n16) -> (ReturnValue:Array of User Widget Object References)) -> (Array Element:Object Reference)
   > n18 = Cast To EWC_PathEntry!(execute, Object=@n17) -> (then, CastFailed, AsEWC Path Entry:EWC Path Entry Object Reference) {is=title K2Node_DynamicCast}
   On_treeview_path_BP_OnGetItemChildren!(
    execute
    self
    Item:Object Reference=Get EntryItemData!(self:EWC Path Entry Object Reference=@n18) -> (EntryItemData:Path Entry Item Data Object Reference)
   ) -> (then, Children:Array of Object References)
   > SetExpanded!(execute, self:EWC Path Entry Object Reference=@n18, IsExpanded:bool=true) -> (then)
   n19 = ClearSelectionFull!(execute, self) -> (then, Assets:Tile View Object Reference) {is=title K2Node_CallFunction}
   n20 = "On Item Selection Changed (treeview_path)"!() -> (OutputDelegate, then, Item:Object Reference, bIsSelected:bool) {is=title K2Node_ComponentBoundEvent}
   n21 = Branch(execute=[@n19.then, @n20.then], Condition=@n20.bIsSelected)
   > n22 = Cast To PathEntryItemData!(execute, Object=@n20.Item) -> (then, CastFailed, AsPath Entry Item Data:Path Entry Item Data Object Reference) {is=title K2Node_DynamicCast}
   > n23 = Execute Python Script!(
    execute
    PythonScript="import unreal\r\n\r\nv_path = unreal.resolve_python_object_handle(path_handle)\r\npure_path = v_path.path.as_posix()\r\n"
    path_handle:Python Object Handle Object Reference=Get PurePathObject!(self:Path Entry Item Data Object Reference=@n22) -> (PurePathObject:Python Object Handle Object Reference)
   ) -> (then, ReturnValue, pure_path:String) {inputs [path_handle], outputs [pure_path]}
   n24 = Branch(
    execute=[@n72.then, @n73.then, @n93.then, @n94.then, @n106.then, @n107.then, @n108.then, @n109.then, @n23.then, @n21.else, @n105.then, @n104.then]
    Condition=Get is_advanced_search() -> (is_advanced_search:bool)
   )
   n25 = "To String (Text)"(InText=Get Text(self=Get SearchByTags() -> (SearchByTags:Editor Utility Editable Text Box Object Reference)))
   n26 = Get MinSearchText() -> (MinSearchText:int=0)
   > n27 = Branch(Condition="integer >= integer"(A=Len(S=@n25), B=@n26))
   n28 = Get search_params_ui() -> (search_params_ui:Map of Strings to Strings)
   > n29 = Add!(execute, TargetMap:Map of Strings to Strings=@n28, Key:"String (by ref)"=tags, Value:"String (by ref)"=@n25) -> (then)
   n30 = "To String (Text)"(InText=Get Text(self=Get SearchBox_Author() -> (SearchBox_Author:Editor Utility Editable Text Box Object Reference)))
   n31 = Branch(execute=[@n39.then, @n29.then], Condition="integer >= integer"(A=Len(S=@n30), B=@n26))
   > n32 = Add(TargetMap=@n28, Key=author, Value=@n30)
   n33 = "To String (Text)"(InText=Get Text(self=Get SearchBox() -> (SearchBox:Editor Utility Editable Text Box Object Reference)))
   n34 = Branch(execute=[@n38.then, @n32.then, @n24.else], Condition="integer >= integer"(A=Len(S=@n33), B=@n26))
   > n35 = Add(TargetMap=@n28, Key=name, Value=@n33)
   n36 = GetSearchParams!(execute=[@n35.then, @n37.then], self) -> (then, SearchParams:Map of Strings to Strings)
   > FilterAssets!(
    execute
    self:Anim Lib Manager Object Reference=Get library()
    searchParams:Map of Strings to Strings=@n36
    Path:String=SelectString(A=@n23.pure_path, bPickA=@n20.bIsSelected)
    use_all:bool=IsChecked(self=Get chk_with_all() -> (chk_with_all:Editor Utility Check Box Object Reference))
    filter_by_rigs:bool=IsChecked(self=Get chk_filterbyrigs() -> (chk_filterbyrigs:Editor Utility Check Box Object Reference))
    ShowAnim:bool=IsChecked(self=Get chkFilterAnim() -> (chkFilterAnim:Editor Utility Check Box Object Reference))
    ShowPose:bool=IsChecked(self=Get chkFilterPose() -> (chkFilterPose:Editor Utility Check Box Object Reference))
   ) -> (then)
   > refreshAssets(ResetTree=false, CalledBy=Search, SendNotification=false, Reconciliations=Get Reconciliations())
   n37 = Remove!(execute=@n34.else, TargetMap:Map of Strings to Strings=@n28, Key:"String (by ref)"=name) -> (then, ReturnValue:bool)
   n38 = Remove(execute=@n31.else, TargetMap=@n28, Key=author)
   n39 = Remove(execute=@n27.else, TargetMap=@n28, Key=tags)
   n40 = "On Item Selection Changed (Assets)"!() -> (OutputDelegate, then, Item:Object Reference, bIsSelected:bool) {is=title K2Node_ComponentBoundEvent}
   > n41 = Branch(Condition=@n40.bIsSelected)
   > n42 = Cast To AnimLibAsset!(execute, Object=@n40.Item) -> (then, CastFailed, AsAnim Lib Asset:Anim Lib Asset Object Reference) {is=title K2Node_DynamicCast}
   > n43 = Set Selected_Asset(Selected_Asset=@n42)
   > Set Selected_Asset_ID(Selected_Asset_ID:String=Get id!(self:Anim Lib Asset Object Reference=@n43) -> (id:String)) -> (Output_Get:String)
   > n44 = Call Thumbnail Updated!(execute:Exec, self:Anim Lib Asset Object Reference=@n43, ThumbPath:String, Task:Automation Editor Task Object Reference) -> (then:Exec) {is=title K2Node_CallDelegate}
   SetIsEnabled(execute=[@n44.then, @n54.then], self=Get PasteActions(), bInIsEnabled=IsValid(Object=@n43))
   n45 = Get Selected_Asset()
   n46 = Get Type!(self:Anim Lib Asset Object Reference=@n45) -> (Type:String)
   n47 = Get PasteFrameRange() -> (PasteFrameRange:EUW Context Menu Frame Range Object Reference)
   > SetIsEnabled(self=@n47, bInIsEnabled=NOT Boolean(A="Equal Exactly (String)"(A=@n46, B=Pose)))
   n48 = Get FrameEnd!(self:EUW Context Menu Frame Range Object Reference=@n47) -> (FrameEnd:Fort Creative Property Numeric Entry Object Reference)
   > n49 = Set MaxValue!(
    execute
    MaxValue:float="To Float (Integer)"(InInt=Get End!(self:Anim Lib Asset Object Reference=@n45) -> (End:int=0))
    self:Spin Box Object Reference=@n48
   ) -> (then, Output_Get:float)
   > n50 = Set Value(Value=@n49, self:Spin Box Object Reference=@n48)
   n51 = Get FrameStart!(self:EUW Context Menu Frame Range Object Reference=@n47) -> (FrameStart:Fort Creative Property Numeric Entry Object Reference)
   > Set MaxValue(MaxValue=@n50, self=@n51)
   > n52 = Set Value(Value="To Float (Integer)"(InInt=Get Start!(self:Anim Lib Asset Object Reference=@n45) -> (Start:int=0)), self=@n51)
   > n53 = Set MinValue!(execute, MinValue:float=@n52, self:Spin Box Object Reference=@n51) -> (then, Output_Get:float)
   > Set MinValue(MinValue=@n53, self=@n48)
   Set Selected_Asset(execute=@n41.else)
   > n54 = Set Selected_Asset_ID()
   "On Clicked (btnSaveAll)"() {is=title K2Node_ComponentBoundEvent}
   n55 = Get library()
   > n56 = GetManager!(execute, self:Anim Lib Manager Object Reference=@n55) -> (then, GetManager:Python Object Handle Object Reference)
   > n57 = "For Each Loop (Map)"!(
    execute:Exec
    MapPin:Map of Strings to Anim Lib Asset Object References=Get LibAssets!(self:Anim Lib Manager Object Reference=@n55) -> (LibAssets:Map of Strings to Anim Lib Asset Object References)
    BreakPin:Exec
   ) -> (then:Exec, KeyPin:String, ValuePin:Anim Lib Asset Object Reference, CompletedPin:Exec)
   > Branch(Condition=Get dirty!(self:Anim Lib Asset Object Reference=@n57.ValuePin) -> (dirty:bool))
   > StartSavingNotification(self:Anim Lib Asset Object Reference=@n57.ValuePin)
   Execute Python Script!(
    execute=@n57.CompletedPin
    PythonScript="import unreal\r\n\r\nmanager = unreal.resolve_python_object_handle(in_manager)\r\nsaved = manager.save_all()"
    in_manager:Python Object Handle Object Reference=@n56
   ) -> (then, ReturnValue) {inputs [in_manager]}
   n58 = "On Selection Changed (cbSelectionSet)"!() -> (OutputDelegate, then, SelectedItem:String, SelectionType:ESelectInfo Enum) {is=title K2Node_ComponentBoundEvent}
   n59 = Find(
    TargetMap:Map of Strings to Controls Selection Set Object References=Get SelectionSets!(self:Anim Lib Manager Object Reference=Get library()) -> (SelectionSets:Map of Strings to Controls Selection Set Object References)
    Key=@n58.SelectedItem
   ) -> (Value:Controls Selection Set Object Reference)
   > Branch(Condition=@n59.ReturnValue)
   > n60 = GetCurrentAssetRigInSequencer()
   > Select!(execute, self:Controls Selection Set Object Reference=@n59, Rig:"Control Rig Object Reference (by ref)"=@n60, Clear:bool=true) -> (then) {is=title K2Node_CallFunction}
   "Main Button Clicked (SaveActions)"() {is=title K2Node_ComponentBoundEvent}
   > n61 = Construct Object from Class(Class=/AnimationLibrary/BP/EUW_dlgNewAsset.EUW_dlgNewAsset_C) -> (ReturnValue:EUW Dlg New Asset Object Reference)
   n62 = Get Selected Item!(self:List View Object Reference=Get treeview_path()) -> (ReturnValue:Object Reference) {is=title K2Node_CallFunction}
   > n63 = Cast To PathEntryItemData(Object=@n62)
   Set CurrentPath!(
    execute=[@n63.CastFailed, @n63.then]
    CurrentPath:String=SelectString(A=Get FullPath!(self:Path Entry Item Data Object Reference=@n63) -> (FullPath:String), bPickA=IsValid(Object=@n62))
    self:EUW Dlg New Asset Object Reference=@n61
   ) -> (then, Output_Get:String)
   > Set Sources!(
    execute
    Sources:Array of Python Object Handle Object References=Get Sources!(self:Anim Lib Manager Object Reference=Get library()) -> (Sources:Array of Python Object Handle Object References)
    self:EUW Dlg New Asset Object Reference=@n61
   ) -> (then, Output_Get:Array of Python Object Handle Object References)
   > n64 = ShowModalEditorUtilityDialogInstance!(
    execute
    WidgetInstance:Editor Utility Dialog Widget Object Reference=@n61
    Title:Text="LOCTEXT(\"New Anim Asset\")"
    ButtonLayout:EAppMsgType Enum=OkCancel
    Size:Vector 2D Structure
   ) -> (then, ReturnValue:Editor Utility Dialog Result Structure)
   n65 = Break Editor Utility Dialog Result!(EditorUtilityDialogResult:"Editor Utility Dialog Result Structure (by ref)"=@n64) -> (ReturnType:EAppReturnType Enum=~, WidgetInstance:Editor Utility Dialog Widget Object Reference) {is=title K2Node_BreakStruct}
   > n66 = Switch on EAppReturnType(Selection=@n65)
   n67 = Cast To EUW_dlgNewAsset!(execute=@n66.Ok, Object=@n65.WidgetInstance) -> (then, CastFailed, AsEUW Dlg New Asset:EUW Dlg New Asset Object Reference) {is=title K2Node_DynamicCast}
   n68 = Get asset!(self:EUW Dlg New Asset Object Reference=@n67) -> (asset:Anim Lib Asset Object Reference)
   > MakeAsset!(
    execute
    self
    AsPose:bool="Equal Exactly (String)"(A=Get Type(self=@n68), B=Pose)
    AsNew:bool=true
    TempAsset:Anim Lib Asset Object Reference=@n68
   ) -> (then, Asset:Anim Lib Asset Object Reference)
   "On Value Changed (sld_ThumbSizer)"!() -> (OutputDelegate, then, Value:float) {is=title K2Node_ComponentBoundEvent}
   n69 = Get Value!(self:Slider Object Reference=Get sld_ThumbSizer()) -> (Value:float)
   > n70 = SetTileSize!(execute, self:Anim Lib Manager Object Reference=Get library(), New Size:double=@n69) -> (then)
   OnTileSizeChanged!(execute=[@n70.then, @n92.then, @n95.then], self, Scale:double=@n69) -> (then)
   "On Clicked (btnSettings)"() {is=title K2Node_ComponentBoundEvent}
   > n71 = Construct Object from Class(Class=/AnimationLibrary/BP/EUW_AnimLibrary_Tab_Settings.EUW_AnimLibrary_Tab_Settings_C) -> (ReturnValue:EUW Anim Library Tab Settings Object Reference)
   > Set parent!(execute, parent:EUW Anim Library Object Reference=Self-Reference(), self:EUW Anim Library Tab Settings Object Reference=@n71) -> (then, Output_Get:EUW Anim Library Object Reference)
   > ShowModalEditorUtilityDialogInstance(WidgetInstance=@n71, Title="LOCTEXT(\"Library Settings\")", ButtonLayout=OkCancel, Size=(X=0,Y=0))
   n72 = "OnTextChanged (Editable Text) (SearchBox_Author)"!() -> (OutputDelegate, then, Text:"Text (by ref)") {is=title K2Node_ComponentBoundEvent}
   n73 = "On Check State Changed (chk_with_all)"!() -> (OutputDelegate, then, bIsChecked:bool) {is=title K2Node_ComponentBoundEvent}
   n74 = "Main Button Clicked (PasteActions)"() {is=title K2Node_ComponentBoundEvent}
   n75 = Get Selected_Asset()
   n76 = is_valid!(execute=[@n74.then, @n96.then, @n115.then], self:Anim Lib Asset Object Reference=@n75) -> (then, valid:bool)
   > n77 = Branch(Condition=@n76)
   > n78 = SelectedBindingHasRig!(execute, self) -> (then, InvalidSelection:bool, HasRig:bool, Selected Binding:Movie Scene Binding Proxy Structure)
   > n79 = Branch(Condition=@n78)
   > LogString(InString=Binding no rig)
   n80 = Branch(execute=@n79.else, Condition=@n78.HasRig)
   n81 = Get AssetHandle!(self:Anim Lib Asset Object Reference=@n75) -> (AssetHandle:Python Object Handle Object Reference)
   n82 = Get PasteFrameRange()
   Execute Python Script!(
    execute=[@n87.then, @n80.then]
    PythonScript="import unreal\r\nfrom AnimLibrary.core.libanim.rig import CRig\r\n\r\ncontext = {'controls': [], 'from_start':  in_start, 'from_end': in_end, 'sections': [], 'original_range': in_original_range}\r\nif in_controls:\r\n    context['controls'] = list(in_controls)\r\nif in_sections:\r\n    context['sections'] = list(in_sections)\r\n\r\nc_rig = CRig.get_selected()\r\nasset = unreal.resolve_python_object_handle(asset_handle)\r\nmanager  = unreal.resolve_python_object_handle(in_manager)\r\nmanager.apply(asset, context)\r\n\r\n"
    asset_handle:Python Object Handle Object Reference=@n81
    in_manager=Get Manager!(self:Anim Lib Manager Object Reference=Get library()) -> (Manager:Python Object Handle Object Reference)
    in_controls:Array of Strings=Get SelectedControls() -> (SelectedControls:Array of Strings)
    in_sections:Array of Strings=Get SelectedSections() -> (SelectedSections:Array of Strings)
    in_start:float=Get Value(self:Spin Box Object Reference=Get FrameStart(self=@n82))
    in_end:float=Get Value(self=Get FrameEnd(self=@n82))
    in_original_range:bool=Get UseOriginalRange() -> (UseOriginalRange:bool)
   ) -> (then, ReturnValue) {inputs [asset_handle, in_manager, in_controls, in_sections, in_start, in_end, in_original_range]}
   n83 = Execute Python Script!(
    execute=@n80.else
    PythonScript="import unreal\r\nunreal.log('PASTE')\r\nasset = unreal.resolve_python_object_handle(asset_handle)\r\nrig_path = asset.metadata.rig_info['rig']\r\nrig_class = None\r\nrig = unreal.load_asset(rig_path)\r\n\r\nif rig:\r\n    \trig_class = rig.get_control_rig_asset_reference()"
    asset_handle=@n81
   ) -> (then, ReturnValue, rig_class:"Control Rig Asset Strong Reference Structure (by ref)", rig_path:String) {inputs [asset_handle], outputs [rig_class,rig_path]}
   > n84 = FindAssetData!(execute, AssetPath:String=@n83.rig_path) -> (then, ReturnValue:Asset Data Structure)
   > Branch(Condition=Is Valid Asset Data!(InAssetData:"Asset Data Structure (by ref)"=@n84) -> (ReturnValue:bool))
   > n85 = GetCurrentLevelSequence!(execute) -> (then, ReturnValue:Level Sequence Object Reference)
   > n86 = GetEditorWorld(self=Get UnrealEditorSubsystem())
   > FindOrCreateControlRigTrackFromReference!(
    execute
    World:World Object Reference=@n86
    LevelSequence:Level Sequence Object Reference=@n85
    InRef:"Control Rig Asset Strong Reference Structure (by ref)"=@n83.rig_class
    InBinding:"Movie Scene Binding Proxy Structure (by ref)"=@n78.Selected Binding
    bIsLayeredControlRig:bool=false
   ) -> (then, ReturnValue:Movie Scene Track Object Reference)
   > n87 = SelectBindings!(
    execute
    ObjectBindings:Array of Movie Scene Binding Proxy Structures=Make Array("[0]":Movie Scene Binding Proxy Structure=@n78.Selected Binding) -> (Array:Array of Movie Scene Binding Proxy Structures)
   ) -> (then)
   LogString(execute=@n77.else, InString=Asset Null)
   "On Clicked (btnExpander)"() {is=title K2Node_ComponentBoundEvent}
   n88 = Get LeftDivider() -> (LeftDivider:EUW Resize Divider Object Reference)
   n89 = "Equal (Enum)"(
    A:ESlateVisibility Enum=Get Visibility!(self:Widget Object Reference=@n88) -> (Visibility:ESlateVisibility Enum=Visible)
    B:ESlateVisibility Enum=Visible
   )
   > n90 = Set Visibility(Visibility=Select(Option 0=Visible, Option 1=Collapsed, Index=@n89) {is=title K2Node_Select}, self=@n88)
   > Set Visibility(Visibility=@n90, self=Get LeftSide() -> (LeftSide:Vertical Box Object Reference))
   > SetBrushFromTexture!(
    execute
    self:Image Object Reference=Get img_expander() -> (img_expander:Image Object Reference)
    Texture:Texture 2D Object Reference=Select(
     Option 0:Texture 2D Object Reference=/AnimationLibrary/resources/icons/anglesleftsolid
     Option 1:Texture 2D Object Reference=/AnimationLibrary/resources/icons/anglesrightsolid
     Index=NOT Boolean(A=@n89)
    ) -> (ReturnValue:Texture 2D Object Reference)
    bMatchSize:bool=false
   ) -> (then)
   "On Clicked (btnRefresh)"() {is=title K2Node_ComponentBoundEvent}
   > shut_down(self:Anim Lib Manager Object Reference=Get library())
   > InitLibrary()
   n91 = Get Reconciliations()
   > refreshAssets(ResetTree=true, CalledBy=RefreshButton, SendNotification=true, Reconciliations=@n91)
   n92 = "On Entries Generated (AssetsAsList)"!() -> (OutputDelegate, then, NumEntries:int=0) {is=title K2Node_ComponentBoundEvent}
   "On Entries Generated (treeview_path)"!() -> (OutputDelegate, then, NumEntries:int=0) {is=title K2Node_ComponentBoundEvent}
   > ExpandAll(self:Tree View Object Reference=@n16)
   "On Clicked (btnHelp)"() {is=title K2Node_ComponentBoundEvent}
   > Execute Python Script!(
    execute
    PythonScript="import webbrowser\r\n\r\nwebbrowser.open(url)"
    url:String=MakeLiteralString(Value=https://epic.enterprise.slack.com/archives/C09FJPMTWGG)
   ) -> (then, ReturnValue) {inputs [url]}
   n93 = "On Check State Changed (chkFilterPose)"!() -> (OutputDelegate, then, bIsChecked:bool) {is=title K2Node_ComponentBoundEvent}
   n94 = "On Check State Changed (chkFilterAnim)"!() -> (OutputDelegate, then, bIsChecked:bool) {is=title K2Node_ComponentBoundEvent}
   n95 = "On Entries Generated (Assets)"!() -> (OutputDelegate, then, NumEntries:int=0) {is=title K2Node_ComponentBoundEvent}
   n96 = "On Item Double Clicked (Assets)"!() -> (OutputDelegate, then, Item:Object Reference) {is=title K2Node_ComponentBoundEvent}
   "On Clicked (btnCreateFolder)"() {is=title K2Node_ComponentBoundEvent}
   n97 = Get treeview_path()
   n98 = Get Selected Item(self=@n97)
   > Is Valid(InputObject=@n98)
   > SetItemExpansion!(execute, self:Tree View Object Reference=@n97, Item:Object Reference=@n98, bExpandItem:bool=false) -> (then)
   > n99 = Cast To PathEntryItemData(Object=@n98)
   > create_folder!(execute, self:Path Entry Item Data Object Reference=@n99, NewFolder:String=NewFolder) -> (then)
   > SetItemExpansion(self=Get treeview_path(), Item=@n98, bExpandItem=true)
   n100 = "On Item Double Clicked (treeview_path)"!() -> (OutputDelegate, then, Item:Object Reference) {is=title K2Node_ComponentBoundEvent}
   > n101 = Cast To PathEntryItemData(Object=@n100.Item)
   > OpenSourcePath(self:Path Entry Item Data Object Reference=@n101)
   "On Clicked (btnAdvancedSearch)"() {is=title K2Node_ComponentBoundEvent}
   n102 = Get VB_Advanced() -> (VB_Advanced:Vertical Box Object Reference)
   > n103 = Branch(Condition=IsVisible(self=@n102))
   > Set Visibility(Visibility=Collapsed, self=@n102)
   > n104 = Set is_advanced_search(is_advanced_search:bool=false) -> (Output_Get:bool)
   Set Visibility(execute=@n103.else, Visibility=Visible, self=@n102)
   > n105 = Set is_advanced_search(is_advanced_search=true)
   n106 = "OnTextChanged (Editable Text) (SearchByTags)"!() -> (OutputDelegate, then, Text:"Text (by ref)") {is=title K2Node_ComponentBoundEvent}
   n107 = "On Check State Changed (chk_filterbyrigs)"!() -> (OutputDelegate, then, bIsChecked:bool) {is=title K2Node_ComponentBoundEvent}
   n108 = "On Property Changed (SearchParamsView)"!() -> (OutputDelegate, then, PropertyName:Name=None) {is=title K2Node_ComponentBoundEvent}
   n109 = "On Text Changed (SearchBox)"!() -> (OutputDelegate, then, Text:"Text (by ref)") {is=title K2Node_ComponentBoundEvent}
   "On Clicked (btnNewSet)"() {is=title K2Node_ComponentBoundEvent}
   > n110 = ShowModalEditorUtilityDialog!(
    execute
    WidgetClass:Editor Utility Dialog Widget Class Reference=/AnimationLibrary/BP/dlgNewSelectionSet.dlgNewSelectionSet_C
    Title:Text="LOCTEXT(\"New SelectionSet\")"
    ButtonLayout:EAppMsgType Enum=OkCancel
    Size:Vector 2D Structure
   ) -> (then, ReturnValue:Editor Utility Dialog Result Structure)
   n111 = Break Editor Utility Dialog Result(EditorUtilityDialogResult=@n110) -> (ReturnType=~)
   > n112 = Switch on EAppReturnType(Selection=@n111)
   PrintString(execute=@n112.Cancel, InString=Cancelled)
   n113 = Cast To dlgNewSelectionSet!(execute=@n112.Ok, Object=@n111.WidgetInstance) -> (then, CastFailed, AsDlg New Selection Set:Dlg New Selection Set Object Reference) {is=title K2Node_DynamicCast}
   n114 = Get edtSetName!(self:Dlg New Selection Set Object Reference=@n113) -> (edtSetName:Editor Utility Editable Text Box Object Reference)
   > CreateSelectionSet!(execute, self, name:String="To String (Text)"(InText=Get Text(self=@n114))) -> (then, new_set:Controls Selection Set Object Reference) {is=title K2Node_CallFunction}
   n115 = OnAssetContextApply() {is=title K2Node_CustomEvent}
   OnSourceChanged() {is=title K2Node_CustomEvent}
   > shut_down(self=Get library())
   > InitLibrary()
   > refreshAssets(ResetTree=true, CalledBy=SourceChanged, SendNotification=true, Reconciliations=@n91)
   Event Destruct()
   > n116 = For Each Loop(Array:Array of User Widget Object References=GetDisplayedEntryWidgets(self=Get Assets())) -> (Array Element:User Widget Object Reference)
   shut_down(execute=@n116.Completed, self=Get library())
   n117 = Cast To EWC_AnimEntry!(execute=@n116.LoopBody, Object=@n116) -> (then, CastFailed, AsEWC Anim Entry:EWC Anim Entry Object Reference) {is=title K2Node_DynamicCast}
   n118 = Get mediaplayer!(self:Image Scrub Object Reference=Get ThumbImg!(self:EWC Anim Entry Object Reference=@n117) -> (ThumbImg:Image Scrub Object Reference)) -> (mediaplayer:Media Player Object Reference)
   > Close(self:Media Player Object Reference=@n118)
   > RemoveAt!(
    execute
    self:Media Playlist Object Reference=Get Playlist!(self:Media Player Object Reference=@n118) -> (Playlist:Media Playlist Object Reference)
    Index:int=0
   ) -> (then, ReturnValue:bool)
   n119 = Event OnInitialized()
   n120 = GetSelectedBindings!() -> (ReturnValue:Array of Movie Scene Binding Proxy Structures)
   Branch(Condition="Equal (Integer)"(A=Length(TargetArray:Array of Movie Scene Binding Proxy Structures=@n120), B=1))
   > n121 = GetRigsBindings!(execute, self) -> (then, BindingIDs:Array of Guid Structures)
   Contains Item(
    TargetArray:Array of Guid Structures=@n121
    ItemToFind:"Guid Structure (by ref)"=Break Movie Scene Binding Proxy!(MovieSceneBindingProxy:"Movie Scene Binding Proxy Structure (by ref)"="Get (a ref)"(Array:Array of Movie Scene Binding Proxy Structures=@n120, Dimension 1=0) -> (Output:"Movie Scene Binding Proxy Structure (by ref)")) -> (BindingID:Guid Structure, Sequence:Movie Scene Sequence Object Reference) {is=title K2Node_BreakStruct}
   )
   refreshAssets(execute=@n123.then, ResetTree=true, CalledBy=BOUND-event, SendNotification=true, Reconciliations=@n4.Reconciliations)
   SetItemExpansion(execute=@n18.then, self=@n16, Item=@n18, bExpandItem=true)
   Event PreConstruct()
   Event Tick()
   n122 = Set UseOriginalRange(execute=@n14.then, UseOriginalRange:bool=@n14.bIsChecked) -> (Output_Get:bool)
   > Set bIsEnabled!(execute, bIsEnabled:bool=NOT Boolean(A=@n122), self:Widget Object Reference=@n11.ContainerItem) -> (then, Output_Get:bool)
   n123 = AddSimpleEditorNotification!(
    execute=@n4.then
    Text:Text="LOCTEXT(\"Animation Library\")"
    SubText:Text="LOCTEXT(\"Assets refreshed!\")"
    FadeInDuration:float=0.5
    FadeOutDuration:float=2
    ExpireDuration:float=1
   ) -> (then)
  ]}
 GetCurrentAssetRigInSequencer {nodes [
   GetCurrentAssetRigInSequencer() {is=title K2Node_FunctionEntry}
   > Set CurrentRig(CurrentRig:Control Rig Object Reference) -> (Output_Get:Control Rig Object Reference)
   > n1 = GetFocusedLevelSequence!(execute) -> (then, ReturnValue:Level Sequence Object Reference)
   > n2 = GetControlRigs!(execute, LevelSequence:Level Sequence Object Reference=@n1) -> (then, ReturnValue:Array of Control Rig Sequencer Binding Proxy Structures)
   > n3 = For Each Loop with Break!(Exec:Exec, Array:Array of Control Rig Sequencer Binding Proxy Structures=@n2, Break:Exec=@n7.then) -> (LoopBody:Exec, Array Element:Control Rig Sequencer Binding Proxy Structure, Array Index:int, Completed:Exec)
   Return Node!(execute=@n3.Completed, Output_Rig:Control Rig Object Reference=Get CurrentRig() -> (CurrentRig:Control Rig Object Reference))
   n4 = Break Control Rig Sequencer Binding Proxy!(ControlRigSequencerBindingProxy:"Control Rig Sequencer Binding Proxy Structure (by ref)"=@n3) -> (
    Proxy:Movie Scene Binding Proxy Structure
    ControlRig:Control Rig Object Reference
    Track:Movie Scene Control Rig Parameter Track Object Reference
   ) {is=title K2Node_BreakStruct}
   > n5 = Execute Python Script!(
    execute
    PythonScript="import unreal\r\npackage = obj.get_class().get_package()\r\nrig_bp = unreal.load_asset(package.get_path_name())"
    obj:Control Rig Object Reference=@n4.ControlRig
   ) -> (then, ReturnValue, rig_bp:Object Reference) {inputs [obj], outputs [rig_bp]}
   n6 = Get Selected_Asset()
   > Branch(Condition="Equal (Object)"(A=@n5.rig_bp, B=Get Rig!(self:Anim Lib Asset Object Reference=@n6) -> (Rig:Control Rig Blueprint Object Reference)))
   > n7 = Set CurrentRig(CurrentRig=@n4.ControlRig)
  ]}
 GetPathEntryChildren {nodes [
   n1 = GetPathEntryChildren!() -> (then, Items:Array of Path Entry Item Data Object References) {is=title K2Node_FunctionEntry}
   > n2 = For Each Loop with Break(Array:Array of Object References=@n1) -> (Array Element:Object Reference)
   n3 = Get AllItems() -> (AllItems:Array of Path Entry Item Data Object References)
   Return Node!(execute=@n2.Completed, OutItems:Array of Path Entry Item Data Object References=@n3)
   > n4 = Cast To PathEntryItemData(Object=@n2)
   > Add(TargetArray:Array of Path Entry Item Data Object References=@n3, NewItem:"Path Entry Item Data Object Reference (by ref)"=@n4) {is=title K2Node_CallArrayFunction}
   > n5 = GetPathEntryChildren!(
    execute
    self
    Items:Array of Path Entry Item Data Object References=Get children!(self:Path Entry Item Data Object Reference=@n4) -> (children:Array of Path Entry Item Data Object References)
   ) -> (then, OutItems:Array of Path Entry Item Data Object References) {is=title K2Node_CallFunction}
   > Append Array!(execute, TargetArray:Array of Path Entry Item Data Object References=@n3, SourceArray:Array of Path Entry Item Data Object References=@n5) -> (then)
  ]}
 GetRigsBindings {nodes [
   GetRigsBindings() {is=title K2Node_FunctionEntry}
   n1 = Get rig_bindings() -> (rig_bindings:Array of Guid Structures)
   > Clear(TargetArray:Array of Guid Structures=@n1)
   > n2 = GetCurrentLevelSequence()
   > n3 = GetControlRigs(LevelSequence=@n2)
   > n4 = For Each Loop(Array:Array of Control Rig Sequencer Binding Proxy Structures=@n3) -> (Array Element:Control Rig Sequencer Binding Proxy Structure)
   Return Node!(execute=@n4.Completed, BindingIDs:Array of Guid Structures=@n1)
   > Add(
    TargetArray:Array of Guid Structures=@n1
    NewItem:"Guid Structure (by ref)"=Break Movie Scene Binding Proxy(MovieSceneBindingProxy=Break Control Rig Sequencer Binding Proxy(ControlRigSequencerBindingProxy=@n4))
   )
  ]}
 GetSearchParams {nodes [
   GetSearchParams() {is=title K2Node_FunctionEntry}
   n1 = Get FinalParams() -> (FinalParams:Map of Strings to Strings)
   > Clear!(execute, TargetMap:Map of Strings to Strings=@n1) -> (then) {is=title K2Node_CallFunction}
   > n2 = Branch(Condition=Get is_advanced_search())
   n3 = Get search_params_ui()
   > n4 = Keys!(execute, TargetMap:Map of Strings to Strings=@n3) -> (then, Keys:Array of Strings)
   > n5 = For Each Loop(Array:Array of Strings=@n4) -> (Array Element:String)
   Return Node!(execute=[@n9.then, @n8.else, @n5.Completed], SearchParams:Map of Strings to Strings=@n1)
   > Add(TargetMap=@n1, Key=@n5, Value=Find(TargetMap:Map of Strings to Strings=@n3, Key=@n5) -> (Value:String)) {is=title K2Node_CallFunction}
   n6 = Get NameKey() -> (NameKey:String)
   n7 = Find(TargetMap=@n3, Key=@n6)
   n8 = Branch(execute=@n2.else, Condition=@n7.ReturnValue)
   > n9 = Add(TargetMap=@n1, Key=@n6, Value=@n7)
  ]}
 GetSelectedPath {nodes [
   GetSelectedPath() {is=title K2Node_FunctionEntry}
   n1 = Get Selected Item(self=Get treeview_path())
   > n2 = Is Valid(InputObject=@n1)
   Return Node!(execute=@n2.Is Not Valid, NewParam:String)
   > n3 = Cast To PathEntryItemData(Object=@n1)
   > n4 = Execute Python Script!(
    execute
    PythonScript="import unreal\r\n\r\nv_path = unreal.resolve_python_object_handle(path_handle)\r\npure_path = v_path.path.as_posix()\r\n"
    path_handle:Python Object Handle Object Reference=Get PurePathObject(self=@n3)
   ) -> (then, ReturnValue, pure_path:String) {inputs [path_handle], outputs [pure_path]}
   > Return Node(NewParam=@n4.pure_path)
  ]}
 InitLibrary {nodes [
   InitLibrary() {is=title K2Node_FunctionEntry}
   > bootstrap(self:Anim Lib Manager Object Reference=Get library())
   refreshAssets(ResetTree=true, SendNotification=false)
  ]}
 IsSelectedNull {nodes [
   n1 = IsSelectedNull!() -> (then, ItemObject:Object Reference) {is=title K2Node_FunctionEntry}
   > n2 = GetEditorWorld(self=Get UnrealEditorSubsystem())
   > Return Node!(execute, Is Null Item:bool=IsNullItem!(ItemObject:Object Reference=@n1, __WorldContext:Object Reference=@n2) -> (Is Null Item:bool))
  ]}
 MakeAsset {nodes [
   n1 = MakeAsset!() -> (then, AsPose:bool=~, AsNew:bool=~, TempAsset:Anim Lib Asset Object Reference) {is=title K2Node_FunctionEntry}
   n2 = Get Selected_Asset()
   > n3 = is_valid(self=@n2)
   > n4 = Branch(Condition=AND Boolean(A=@n3, B=NOT Boolean(A=@n1.AsNew)))
   > n5 = Show Message Dialog(
    Title=MakeLiteralText(Value="LOCTEXT(\"Animation Library\")")
    Message=MakeLiteralText(Value="LOCTEXT(\"You\\'re about to overwrite this asset.\\r\\nAre your sure?\\r\\nOk: Overwrite selected asset.\\r\\nCancel: Do nothing.\")")
    MessageType=OkCancel
    DefaultValue=Ok
    MessageCategory=Warning
   )
   > n6 = Switch on EAppReturnType(Selection=@n5)
   n7 = Get Manager(self=Get library())
   Execute Python Script!(
    execute=@n6.Ok
    PythonScript="import unreal\r\nfrom AnimLibrary.core.utils.misc import get_sequence\r\n\r\nasset = unreal.resolve_python_object_handle(asset_handle)\r\nmanager  = unreal.resolve_python_object_handle(in_manager)\r\nls = get_sequence()\r\nmanager.extract_to(asset, sequence=ls)\r\n\r\n"
    asset_handle:Python Object Handle Object Reference=Get AssetHandle(self=@n2)
    in_manager:Python Object Handle Object Reference=@n7
   ) -> (then, ReturnValue) {inputs [asset_handle,in_manager]}
   > Return Node!(execute, Asset:Anim Lib Asset Object Reference=@n2)
   n8 = Execute Python Script!(
    execute=@n4.else
    PythonScript="import unreal\r\nfrom AnimLibrary.core.utils.misc import get_sequence\r\nfrom AnimLibrary.core.constants import AssetType\r\n\r\nmanager  = unreal.resolve_python_object_handle(in_manager)\r\n\r\n## use the current sequencer name as default for new assets\r\n\r\nls = get_sequence()\r\nif ls and not new_name:\r\n    new_name = ls.get_name()\r\nasset_type = AssetType.ANIMATION\r\nif as_pose:\r\n   asset_type = AssetType.POSE\r\nasset = manager.create(new_name, asset_type=asset_type, sequence=ls)\r\nasset_handle = unreal.create_python_object_handle(asset)\r\nasset.metadata.note = new_notes\r\nasset.metadata.author = new_author\r\nis_valid = manager.is_valid_path(new_path=new_path)\r\nif is_valid:     \r\n    asset.virtual_path = new_path\r\nasset.mark_dirty()"
    in_manager=@n7
    new_name:String=Get Name!(self:Anim Lib Asset Object Reference=@n1.TempAsset) -> (Name:String)
    as_pose:bool=@n1
    new_author:String=Get Author!(self:Anim Lib Asset Object Reference=@n1.TempAsset) -> (Author:String)
    new_notes:String=Get Notes!(self:Anim Lib Asset Object Reference=@n1.TempAsset) -> (Notes:String)
    new_path:String=Get Path!(self:Anim Lib Asset Object Reference=@n1.TempAsset) -> (Path:String)
   ) -> (then, ReturnValue, asset_handle:Python Object Handle Object Reference) {inputs [in_manager,new_name,as_pose,new_author,new_notes,new_path], outputs [asset_handle]}
   > Set AssetHandle!(execute, AssetHandle:Python Object Handle Object Reference=@n8.asset_handle, self:Anim Lib Asset Object Reference=@n1.TempAsset) -> (then, Output_Get:Python Object Handle Object Reference)
   > update_thumbnail(self=@n1.TempAsset)
   > Return Node()
  ]}
 OnClosedDialog {nodes [
   OnClosedDialog() {is=title K2Node_FunctionEntry}
   > Set bIsEnabled(bIsEnabled=true, self=Get HBMain() -> (HBMain:Horizontal Box Object Reference))
   > Set Visibility(Visibility=Collapsed, self=Get dlgSlot() -> (dlgSlot:Named Slot Object Reference))
  ]}
 OnTileSizeChanged {nodes [
   n1 = OnTileSizeChanged!() -> (then, Scale:double=~) {is=title K2Node_FunctionEntry}
   n2 = Get TileSizeMax() -> (TileSizeMax:double)
   n3 = float * float(A=@n2, B=@n1) {is=title K2Node_PromotableOperator}
   n4 = Get TileSizeMin() -> (TileSizeMin:double)
   n5 = "Clamp (Float)"(Value=@n3, Min=@n4, Max=@n2)
   n6 = Get Assets()
   > SetEntryWidth!(execute, self:Tile View Object Reference=@n6, NewWidth:float=@n5) -> (then)
   > SetEntryHeight!(execute, self:Tile View Object Reference=@n6, NewHeight:float=@n5) -> (then)
   n7 = "float < float"(A=@n3, B=@n4)
   > n8 = Branch(Condition=@n7)
   n9 = Get AssetsAsList()
   > n10 = Set List Items!(
    execute
    self:List View Object Reference=@n9
    InListItems:Array of Object References=GetListItems!(self:List View Object Reference=@n6) -> (ReturnValue:Array of Object References)
   ) -> (then) {is=title K2Node_CallFunction}
   Set Visibility(
    execute=[@n10.then, @n8.else]
    Visibility=Select(Option 0:ESlateVisibility Enum=Collapsed, Option 1:ESlateVisibility Enum=Visible, Index=@n7) -> (ReturnValue:ESlateVisibility Enum)
    self=@n9
   )
   > Set Visibility(Visibility=Select(Option 0=Visible, Option 1=Collapsed, Index=@n7), self=@n6)
   > n11 = For Each Loop(Array:Array of User Widget Object References=Select(
     Option 0:Array of User Widget Object References=GetDisplayedEntryWidgets(self=@n6)
     Option 1:Array of User Widget Object References=GetDisplayedEntryWidgets(self=@n9)
     Index=@n7
    ) -> (ReturnValue:Array of User Widget Object References)) -> (Array Element:User Widget Object Reference)
   > n12 = Cast To EWC_AnimEntry(Object=@n11)
   > ResizeWidget!(execute, self:EWC Anim Entry Object Reference=@n12, Scale:double=@n1) -> (then)
  ]}
 On_treeview_path_BP_OnGetItemChildren {nodes [
   n1 = On_treeview_path_BP_OnGetItemChildren!() -> (then, Item:Object Reference) {is=title K2Node_FunctionEntry}
   > n2 = Cast To PathEntryItemData(Object=@n1)
   > Return Node!(execute, Children:Array of Object References=Get children(self=@n2))
  ]}
 OpenDialog {nodes [
   n1 = OpenDialog!() -> (
    then
    ChildUI:Editor Utility Widget Class Reference
    Title:Text="LOCTEXT(\"MyDialog\")"
    MsgType:EAppMsgType Enum=OkCancel
    MsgCategory:EAppMsgCategory Enum=Info
    OnCloseEvent:"Delegate (by ref)"
   ) {is=title K2Node_FunctionEntry}
   > Set bIsEnabled(bIsEnabled=false, self=Get HBMain())
   n2 = Get dlgSlot()
   > Set Visibility(Visibility=Visible, self=@n2)
   > n3 = Create Widget!(
    execute:Exec
    WorldContextObject:Object Reference
    Class:User Widget Class Reference=/Game/NeverCook/UI/Modal/EUWModalBase.EUWModalBase_C
    OwningPlayer:Player Controller Object Reference
   ) -> (then:Exec, ReturnValue:EUWModal Base Object Reference)
   > SetContent!(execute, self:Content Widget Object Reference=@n2, Content:Widget Object Reference=@n3) -> (then, ReturnValue:Panel Slot Object Reference)
   > n4 = SetupModal!(
    execute
    self:EUWModal Base Object Reference=@n3
    Title:Text=@n1.Title
    AnchorOwner:Menu Anchor Object Reference=Get MenuAnchorDialog() -> (MenuAnchorDialog:Menu Anchor Object Reference)
    ContextOwner:Widget Object Reference=Self-Reference()
    Widget Class:Editor Utility Widget Class Reference=@n1
    Widget Instance:Editor Utility Widget Object Reference
    MsgType:EAppMsgType Enum=@n1.MsgType
    MsgCategory:EAppMsgCategory Enum=@n1.MsgCategory
    Event:"Delegate (by ref)"=@n1.OnCloseEvent
   ) -> (then, Widget:Editor Utility Widget Object Reference)
   > Return Node!(execute, Child Widget:Editor Utility Widget Object Reference=@n4)
  ]}
 PointerInWidget {nodes [
   n1 = PointerInWidget!() -> (then, TargetWidget:Widget Object Reference, PointerEvent:"Pointer Event Structure (by ref)") {is=title K2Node_FunctionEntry}
   n2 = BreakVector2D(InVec=Get Screen Space Position!(Input:"Pointer Event Structure (by ref)"=@n1.PointerEvent) -> (ReturnValue:Vector 2D Structure) {is=title K2Node_CallFunction})
   n3 = GetTickSpaceGeometry!(self:Widget Object Reference=@n1) -> (ReturnValue:"Geometry Structure (by ref)")
   n4 = BreakVector2D(InVec=GetAbsoluteSize!(Geometry:"Geometry Structure (by ref)"=@n3) -> (ReturnValue:Vector 2D Structure))
   n5 = BreakVector2D(InVec=GetAbsolutePosition!(Geometry:"Geometry Structure (by ref)"=@n3) -> (ReturnValue:Vector 2D Structure))
   n6 = float + float(A=@n5.Y, B=@n4.Y)
   n7 = float + float(A=@n5, B=@n4)
   > Return Node!(
    execute
    IsWithin:bool=AND Boolean!(
     A="float >= float"(A=@n2, B=@n5)
     B="float < float"(A=@n2, B=@n7)
     C:bool="float >= float"(A=@n2.Y, B=@n5.Y)
     D:bool="float < float"(A=@n2.Y, B=@n6)
    ) -> (ReturnValue)
   )
  ]}
 PopulateSets {nodes [
   PopulateSets() {is=title K2Node_FunctionEntry}
   n1 = Get cbSelectionSet() -> (cbSelectionSet:"Searchable Combo Box (String) Object Reference")
   > ClearOptions(self:"Searchable Combo Box (String) Object Reference"=@n1)
   > n2 = "For Each Loop (Map)"(MapPin:Map of Strings to Controls Selection Set Object References=Get SelectionSets(self=Get library())) -> (ValuePin:Controls Selection Set Object Reference)
   > AddOption!(execute, self:"Searchable Combo Box (String) Object Reference"=@n1, Option:String=@n2) -> (then)
  ]}
 RefreshVirtualPathTree {nodes [
   RefreshVirtualPathTree() {is=title K2Node_FunctionEntry}
   > n1 = Cast To PathEntryItemData(Object=Get Selected Item(self=Get treeview_path()))
   > n2 = Set currentSelection(currentSelection:String=Get FullPath(self=@n1)) -> (Output_Get:String)
   n3 = Get treeview_path()
   ClearListItems(execute=[@n1.CastFailed, @n2.then], self=@n3)
   > n4 = Execute Python Script!(
    execute
    PythonScript="import unreal\r\n\r\nmanager = unreal.resolve_python_object_handle(manager_handle)\r\n\r\nroot = manager.virtual_tree_root\r\n\r\ntree = [unreal.create_python_object_handle(c) for c in root.children]"
    manager_handle:Python Object Handle Object Reference=Get Manager(self=Get library())
   ) -> (then, ReturnValue, tree:Array of Python Object Handle Object References) {inputs [manager_handle], outputs [tree]}
   > n5 = Set basePaths(basePaths:Array of Python Object Handle Object References=@n4.tree) -> (Output_Get:Array of Python Object Handle Object References)
   > n6 = For Each Loop(Array:Array of Python Object Handle Object References=@n5) -> (Array Element:Python Object Handle Object Reference)
   n7 = Get treeview_path()
   n8 = For Each Loop(Exec=@n6.Completed, Array:Array of Object References=GetListItems(self=@n7)) -> (Array Element:Object Reference)
   n9 = Get AllItems()
   n10 = GetPathEntryChildren(execute=@n8.Completed, Items=@n9)
   > n11 = For Each Loop with Break(Array:Array of Path Entry Item Data Object References=@n10, Break=@n15.then) -> (Array Element:Path Entry Item Data Object Reference)
   n12 = Get FoundSelectedItem() -> (FoundSelectedItem:Object Reference)
   Set CurrentOuter(execute=@n11.Completed, CurrentOuter:Object Reference=@n12) -> (Output_Get:Object Reference)
   n13 = Get CurrentOuter() -> (CurrentOuter:Object Reference)
   > n14 = While Loop!(execute:Exec, Condition:bool=IsValid(Object=@n13)) -> (LoopBody:Exec, Completed:Exec) {is=title K2Node_MacroInstance}
   Set Selected Item(execute=@n14.Completed, self=@n7, Item=@n12)
   SetItemExpansion(execute=@n14.LoopBody, self=@n7, Item=Get CurrentOuter(), bExpandItem=true)
   > Set CurrentOuter(CurrentOuter=GetOuterObject!(Object:Object Reference=@n13) -> (ReturnValue:Object Reference))
   Branch(execute=@n11.LoopBody, Condition="Equal Exactly (String)"(A=Get FullPath(self=@n11), B=Get currentSelection() -> (currentSelection:String)))
   > n15 = Set FoundSelectedItem(FoundSelectedItem:Object Reference=@n11) -> (Output_Get:Object Reference)
   n16 = Cast To PathEntryItemData(execute=@n8.LoopBody, Object=@n8)
   > Add(TargetArray:Array of Path Entry Item Data Object References=@n9, NewItem:"Path Entry Item Data Object Reference (by ref)"=@n16) {is=title K2Node_CallArrayFunction}
   Is Valid(exec=@n6.LoopBody, InputObject=@n6)
   > n17 = Construct Object from Class(Class=/AnimationLibrary/BP/PathEntryItemData.PathEntryItemData_C) -> (ReturnValue:Path Entry Item Data Object Reference)
   > Set PurePathObject!(execute, PurePathObject:Python Object Handle Object Reference=@n6, self:Path Entry Item Data Object Reference=@n17) -> (then, Output_Get:Python Object Handle Object Reference)
   > Set OwnerLibWidget!(execute, OwnerLibWidget:EUW Anim Library Object Reference=Self-Reference(), self:Path Entry Item Data Object Reference=@n17) -> (then, Output_Get:EUW Anim Library Object Reference)
   > On Init(self:Path Entry Item Data Object Reference=@n17)
   > AddItem!(execute, self:List View Object Reference=@n3, Item:Object Reference=@n17) -> (then)
  ]}
 SelectedBindingHasRig {nodes [
   SelectedBindingHasRig() {is=title K2Node_FunctionEntry}
   > Add Items!(
    execute
    TargetSet:Set of Movie Scene Binding Proxy Structures=Get UniqueBindings() -> (UniqueBindings:Set of Movie Scene Binding Proxy Structures)
    NewItems:Array of Movie Scene Binding Proxy Structures=GetSelectedBindings()
   ) -> (then)
   > n1 = GetRigsBindings() {is=title K2Node_CallFunction}
   n2 = Get UniqueBindings()
   n3 = "Equal (Integer)"(A=Length!(TargetSet:Set of Movie Scene Binding Proxy Structures=@n2) -> (ReturnValue:int=0) {is=title K2Node_CallFunction}, B=1)
   > n4 = Branch(Condition=@n3)
   n5 = Get Item By Index!(TargetSet:Set of Movie Scene Binding Proxy Structures=@n2, Index:int=0) -> (Item:Movie Scene Binding Proxy Structure) {is=title K2Node_CallFunction}
   > Return Node!(
    execute
    InvalidSelection:bool=NOT Boolean(A=@n3)
    HasRig:bool=Contains Item(TargetArray=@n1, ItemToFind=Break Movie Scene Binding Proxy(MovieSceneBindingProxy=@n5))
    Selected Binding:"Movie Scene Binding Proxy Structure (by ref)"=@n5
   )
   n6 = GetCurrentLevelSequence(execute=@n4.else)
   > n7 = GetControlRigs(LevelSequence=@n6)
   > n8 = Branch(Condition="Equal (Integer)"(A=Length(TargetArray:Array of Control Rig Sequencer Binding Proxy Structures=@n7) {is=title K2Node_CallArrayFunction}, B=1))
   > Return Node(
    InvalidSelection=false
    HasRig=true
    Selected Binding=Break Control Rig Sequencer Binding Proxy(ControlRigSequencerBindingProxy="Get (a ref)"(Array:Array of Control Rig Sequencer Binding Proxy Structures=@n7, Dimension 1=0) -> (Output:"Control Rig Sequencer Binding Proxy Structure (by ref)"))
   )
   Return Node(execute=@n8.else, InvalidSelection=true, HasRig=false)
  ]}
 SetLibrary {nodes [
   SetLibrary() {is=title K2Node_FunctionEntry}
   > n1 = Construct Object from Class(Class=/AnimationLibrary/BP/AnimLibManager.AnimLibManager_C) -> (ReturnValue:Anim Lib Manager Object Reference)
   > Set library(library:Anim Lib Manager Object Reference=@n1) -> (Output_Get:Anim Lib Manager Object Reference)
   > InitLibrary() {is=title K2Node_CallFunction}
  ]}
 refreshAssets {nodes [
   n1 = refreshAssets!() -> (then, ResetTree:bool=true, CalledBy:String, SendNotification:bool=~, Reconciliations:Map of Strings to Strings) {is=title K2Node_FunctionEntry}
   > n2 = Sequence()
   > n3 = Set CurrentID(CurrentID:String=Get Selected_Asset_ID() -> (Selected_Asset_ID:String)) -> (Output_Get:String)
   > PrintString(InString=Append!(A=@n1.CalledBy, B="  Selected ID: ", C:String=@n3) -> (ReturnValue))
   ClearSelectionFull(execute=@n2.then_1)
   Branch(execute=@n2.then_2, Condition=@n1)
   > RefreshVirtualPathTree() {is=title K2Node_CallFunction}
   n4 = Branch(
    execute=@n2.then_3
    Condition=Is Not Empty!(TargetMap:Map of Strings to Strings=@n1.Reconciliations) -> (ReturnValue:bool) {is=title K2Node_CallFunction}
   )
   > n5 = "For Each Loop (Map)"(MapPin:Map of Strings to Strings=@n1.Reconciliations) -> (ValuePin:String)
   > n6 = Switch on String!(execute:Exec, Selection:String=@n5.ValuePin) -> (Default:Exec, updated:Exec, created:Exec, removed:Exec) {pinNames [updated,created,removed]}
   n7 = Get LibAssets(self=Get library())
   n8 = Find(TargetMap:Map of Strings to Anim Lib Asset Object References=@n7, Key=@n5) -> (Value:Anim Lib Asset Object Reference)
   n9 = Get Assets()
   AddItem(execute=@n6.created, self=@n9, Item=@n8)
   n10 = Get AssetsAsList()
   > AddItem(self=@n10, Item=@n8)
   RemoveItem!(execute=@n6.removed, self:List View Object Reference=@n9, Item:Object Reference=@n8) -> (then)
   > RemoveItem(self=@n10, Item=@n8)
   > Remove(TargetMap:Map of Strings to Anim Lib Asset Object References=@n7, Key=@n5)
   n11 = Values!(execute=@n4.else, TargetMap:Map of Strings to Anim Lib Asset Object References=@n7) -> (then, Values:Array of Anim Lib Asset Object References)
   > Set List Items(self=@n9, InListItems=@n11)
   > Set List Items(self=@n10, InListItems=@n11)
   > RegenerateAllEntries(self=@n9)
   > RegenerateAllEntries(self=@n10)
   n12 = Find(TargetMap=@n7, Key=Get CurrentID() -> (CurrentID:String))
   Branch(execute=@n2.then_4, Condition=@n12.ReturnValue)
   > Set Selected Item(self=@n9, Item=@n12)
   > Set Selected Item(self=@n10, Item=@n12)
   Branch(execute=@n2.then_5, Condition=@n1.SendNotification)
   > AddSimpleEditorNotification(Text="LOCTEXT(\"Animation Library\")", SubText="LOCTEXT(\"Asset refreshed!\")", FadeInDuration=0.5, FadeOutDuration=2, ExpireDuration=1)
  ]}
]
variables [
 Dialog UI {type=Class, default=&EUW_AnimLibrary_Tab_Settings_C}
 In Quick Render Settings {type=MovieGraphQuickRenderModeSettings, default=null}
 is_advanced_search {type=boolean}
 library {type=AnimLibManager_C, default=null}
 MinSearchText {type=integer, default=3}
 PasteFrameRange {type=EUW_ContextMenu_FrameRange_C, default=null}
 Reconciliations {type="map<string>", default {}}
 search_params {type="map<string>", default {}}
 search_params_ui {type="map<string>", default {}}
 Selected_Asset {type=AnimLibAsset_C, default=null}
 Selected_Asset_ID {type=string, default=None}
 SelectedControls {type="array<string>", default []}
 SelectedSections {type="array<string>", default []}
 self_tab_id {type=string, default=None}
 UseOriginalRange {type=boolean}
]
widgets {
 info {rootWidgetClass=&CanvasPanel, widgetCount=66}
 widgets [
  CanvasPanel_0 {widgetClassPath=&CanvasPanel, widget=CanvasPanel_0}
  HBMain {
   widgetClassPath=&HorizontalBox
   bIsVariable=true
   parent=CanvasPanel_0
   widget=HBMain
   properties {bOverride_Cursor=true}
   slot {class=CanvasPanelSlot, properties {layoutData {alignment=<0 0>, anchors {maximum=<1 1>, minimum=<0 0>}, offsets=<ltrb 0 0 8 2>}}}
  }
  LeftSide {
   widgetClassPath=&VerticalBox
   bIsVariable=true
   parent=HBMain
   widget=LeftSide
   properties {visibility=Collapsed}
   slot {class=HorizontalBoxSlot}
  }
  HorizontalBox_2 {widgetClassPath=&HorizontalBox, parent=LeftSide, widget=HorizontalBox_2, slot {class=VerticalBoxSlot}}
  TextBlock_43 {
   widgetClassPath=&TextBlock
   parent=HorizontalBox_2
   widget=TextBlock_43
   properties {
    lineHeightPercentage=0.94
    text=Folders
    colorAndOpacity {colorUseRule=UseColor_Specified, specifiedColor=<rgba 1 1 1 0.5>}
    font {
     bForceMonospaced=false
     bMaterialIsStencil=false
     fontMaterial=null
     fontObject=&Roboto
     letterSpacing=0
     monospacedWidth=1
     size=9
     skewAmount=0
     typefaceFontName=Bold
     outlineSettings {
      bApplyOutlineToDropShadows=false
      bMiteredCorners=false
      bSeparateFillAlpha=false
      outlineMaterial=null
      outlineSize=0
      outlineColor=<rgba 0 0 0 1>
     }
    }
   }
   slot {class=HorizontalBoxSlot, properties {verticalAlignment=VAlign_Center, size {sizeRule=Fill, value=1}}}
  }
  SizeBox_0 {
   widgetClassPath=&SizeBox
   parent=HorizontalBox_2
   widget=SizeBox_0
   properties {
    bOverride_MaxDesiredHeight=true
    bOverride_MaxDesiredWidth=true
    bOverride_MinDesiredHeight=true
    bOverride_MinDesiredWidth=true
    maxDesiredHeight=32
    maxDesiredWidth=32
    minDesiredHeight=16
    minDesiredWidth=16
   }
   slot {class=HorizontalBoxSlot, properties {horizontalAlignment=HAlign_Right}}
  }
  btnCreateFolder {
   widgetClassPath=&EditorUtilityButton
   parent=SizeBox_0
   widget=btnCreateFolder
   properties {
    toolTipText="[NOT IMPLEMENTED YET]\r\nAdds a subfolder to the current selection"
    widgetStyle {
     clickedSlateSound {resourceObject=null}
     disabled {
      bIsDynamicallyLoaded=false
      drawAs=RoundedBox
      imageType=NoImage
      mirroring=NoMirror
      resourceName=None
      resourceObject=null
      tiling=NoTile
      imageSize=<0 0>
      margin=<ltrb 0 0 0 0>
      outlineSettings {
       bUseBrushTransparency=false
       roundingType=FixedRadius
       width=1
       color {colorUseRule=UseColor_ColorTable, specifiedColor=<rgba 0 0 0 0>}
       cornerRadii=<4 4 4 4>
      }
      tintColor {colorUseRule=UseColor_Specified, specifiedColor=<rgba 0.0395 0.0395 0.0395 1>}
      uVRegion {bIsValid=false, max=<0 0>, min=<0 0>}
     }
     disabledForeground {colorUseRule=UseColor_ColorTable, specifiedColor=<rgba 0 0 0 0>}
     hovered {
      bIsDynamicallyLoaded=false
      drawAs=RoundedBox
      imageType=NoImage
      mirroring=NoMirror
      resourceName=None
      resourceObject=null
      tiling=NoTile
      imageSize=<32 32>
      margin=<ltrb 0 0 0 0>
      outlineSettings {
       bUseBrushTransparency=false
       roundingType=FixedRadius
       width=1
       color {colorUseRule=UseColor_ColorTable, specifiedColor=<rgba 0 0 0 0>}
       cornerRadii=<4 4 4 4>
      }
      tintColor {colorUseRule=UseColor_Specified, specifiedColor=<rgba 0.0953 0.0953 0.0953 1>}
      uVRegion {bIsValid=false, max=<0 0>, min=<0 0>}
     }
     hoveredForeground {colorUseRule=UseColor_ColorTable, specifiedColor=<rgba 0 0 0 0>}
     hoveredSlateSound {resourceObject=null}
     normal {
      bIsDynamicallyLoaded=false
      drawAs=RoundedBox
      imageType=NoImage
      mirroring=NoMirror
      resourceName=None
      resourceObject=null
      tiling=NoTile
      imageSize=<32 32>
      margin=<ltrb 0 0 0 0>
      outlineSettings {
       bUseBrushTransparency=false
       roundingType=FixedRadius
       width=1
       color {colorUseRule=UseColor_ColorTable, specifiedColor=<rgba 0 0 0 0>}
       cornerRadii=<4 4 4 4>
      }
      tintColor {colorUseRule=UseColor_Specified, specifiedColor=<rgba 0.0395 0.0395 0.0395 1>}
      uVRegion {bIsValid=false, max=<0 0>, min=<0 0>}
     }
     normalForeground {colorUseRule=UseColor_ColorTable, specifiedColor=<rgba 0 0 0 0>}
     normalPadding=<ltrb 4 4 4 4>
     pressed {
      bIsDynamicallyLoaded=false
      drawAs=RoundedBox
      imageType=NoImage
      mirroring=NoMirror
      resourceName=None
      resourceObject=null
      tiling=NoTile
      imageSize=<32 32>
      margin=<ltrb 0 0 0 0>
      outlineSettings {
       bUseBrushTransparency=false
       roundingType=FixedRadius
       width=1
       color {colorUseRule=UseColor_ColorTable, specifiedColor=<rgba 0 0 0 0>}
       cornerRadii=<4 4 4 4>
      }
      tintColor {colorUseRule=UseColor_Specified, specifiedColor=<rgba 0.0284 0.0284 0.0284 1>}
      uVRegion {bIsValid=false, max=<0 0>, min=<0 0>}
     }
     pressedForeground {colorUseRule=UseColor_ColorTable, specifiedColor=<rgba 0 0 0 0>}
     pressedPadding=<ltrb 4 4.5 4 3.5>
     pressedSlateSound {resourceObject=null}
    }
   }
   slot {class=SizeBoxSlot}
  }
  imgAddFolderbuttonIcon {
   widgetClassPath=&Image
   parent=btnCreateFolder
   widget=imgAddFolderbuttonIcon
   properties {brush {
     bIsDynamicallyLoaded=false
     drawAs=Image
     imageType=FullColor
     mirroring=NoMirror
     resourceName=None
     resourceObject=/AnimationLibrary/resources/icons/folderplussolid
     tiling=NoTile
     imageSize=<20 20>
     margin=<ltrb 0 0 0 0>
     outlineSettings {
      bUseBrushTransparency=false
      roundingType=HalfHeightRadius
      width=0
      color {colorUseRule=UseColor_Specified, specifiedColor=<rgba 0 0 0 0>}
      cornerRadii=<0 0 0 0>
     }
     tintColor {colorUseRule=UseColor_Specified, specifiedColor=<rgba 1 1 1 1>}
     uVRegion {bIsValid=false, max=<0 0>, min=<0 0>}
    }}
   slot {class=ButtonSlot, properties {padding=<ltrb 0 0 0 0>}}
  }
  treeview_path {
   widgetClassPath=/Script/CommonUI.CommonTreeView
   bIsVariable=true
   parent=LeftSide
   widget=treeview_path
   properties {
    bAllowDragDrop=true
    bClearSelectionOnClick=true
    bP_OnGetItemChildren="(null).On_treeview_path_BP_OnGetItemChildren"
    entryWidgetClass=/AnimationLibrary/BP/EWC_PathEntry.EWC_PathEntry_C
    numDesignerPreviewEntries=10
    verticalEntrySpacing=4
    scrollBarStyle {
     thickness=8
     draggedThumbImage {
      bIsDynamicallyLoaded=false
      drawAs=RoundedBox
      imageType=NoImage
      mirroring=NoMirror
      resourceName=None
      resourceObject=null
      tiling=NoTile
      imageSize=<0 0>
      margin=<ltrb 0 0 0 0>
      outlineSettings {
       bUseBrushTransparency=false
       roundingType=FixedRadius
       width=0
       color {colorUseRule=UseColor_Specified, specifiedColor=<rgba 0 0 0 0>}
       cornerRadii=<4 4 4 4>
      }
      tintColor {colorUseRule=UseColor_Specified, specifiedColor=<rgba 0.2159 0.2159 0.2159 1>}
      uVRegion {bIsValid=false, max=<0 0>, min=<0 0>}
     }
     horizontalBackgroundImage {
      bIsDynamicallyLoaded=false
      drawAs=NoDrawType
      imageType=NoImage
      mirroring=NoMirror
      resourceName=None
      resourceObject=null
      tiling=NoTile
      imageSize=<0 0>
      margin=<ltrb 0 0 0 0>
      outlineSettings {
       bUseBrushTransparency=false
       roundingType=HalfHeightRadius
       width=0
       color {colorUseRule=UseColor_Specified, specifiedColor=<rgba 0 0 0 0>}
       cornerRadii=<0 0 0 0>
      }
      tintColor {colorUseRule=UseColor_Specified, specifiedColor=<rgba 1 1 1 1>}
      uVRegion {bIsValid=false, max=<0 0>, min=<0 0>}
     }
     horizontalBottomSlotImage {
      bIsDynamicallyLoaded=false
      drawAs=NoDrawType
      imageType=NoImage
      mirroring=NoMirror
      resourceName=None
      resourceObject=null
      tiling=NoTile
      imageSize=<0 0>
      margin=<ltrb 0 0 0 0>
      outlineSettings {
       bUseBrushTransparency=false
       roundingType=HalfHeightRadius
       width=0
       color {colorUseRule=UseColor_Specified, specifiedColor=<rgba 0 0 0 0>}
       cornerRadii=<0 0 0 0>
      }
      tintColor {colorUseRule=UseColor_Specified, specifiedColor=<rgba 1 1 1 1>}
      uVRegion {bIsValid=false, max=<0 0>, min=<0 0>}
     }
     horizontalTopSlotImage {
      bIsDynamicallyLoaded=false
      drawAs=NoDrawType
      imageType=NoImage
      mirroring=NoMirror
      resourceName=None
      resourceObject=null
      tiling=NoTile
      imageSize=<0 0>
      margin=<ltrb 0 0 0 0>
      outlineSettings {
       bUseBrushTransparency=false
       roundingType=HalfHeightRadius
       width=0
       color {colorUseRule=UseColor_Specified, specifiedColor=<rgba 0 0 0 0>}
       cornerRadii=<0 0 0 0>
      }
      tintColor {colorUseRule=UseColor_Specified, specifiedColor=<rgba 1 1 1 1>}
      uVRegion {bIsValid=false, max=<0 0>, min=<0 0>}
     }
     hoveredThumbImage {
      bIsDynamicallyLoaded=false
      drawAs=RoundedBox
      imageType=NoImage
      mirroring=NoMirror
      resourceName=None
      resourceObject=null
      tiling=NoTile
      imageSize=<0 0>
      margin=<ltrb 0 0 0 0>
      outlineSettings {
       bUseBrushTransparency=false
       roundingType=FixedRadius
       width=0
       color {colorUseRule=UseColor_Specified, specifiedColor=<rgba 0 0 0 0>}
       cornerRadii=<4 4 4 4>
      }
      tintColor {colorUseRule=UseColor_Specified, specifiedColor=<rgba 0.2159 0.2159 0.2159 1>}
      uVRegion {bIsValid=false, max=<0 0>, min=<0 0>}
     }
     normalThumbImage {
      bIsDynamicallyLoaded=false
      drawAs=RoundedBox
      imageType=NoImage
      mirroring=NoMirror
      resourceName=None
      resourceObject=null
      tiling=NoTile
      imageSize=<0 0>
      margin=<ltrb 0 0 0 0>
      outlineSettings {
       bUseBrushTransparency=false
       roundingType=FixedRadius
       width=0
       color {colorUseRule=UseColor_Specified, specifiedColor=<rgba 0 0 0 0>}
       cornerRadii=<4 4 4 4>
      }
      tintColor {colorUseRule=UseColor_Specified, specifiedColor=<rgba 0.0953 0.0953 0.0953 1>}
      uVRegion {bIsValid=false, max=<0 0>, min=<0 0>}
     }
     verticalBackgroundImage {
      bIsDynamicallyLoaded=false
      drawAs=NoDrawType
      imageType=NoImage
      mirroring=NoMirror
      resourceName=None
      resourceObject=null
      tiling=NoTile
      imageSize=<0 0>
      margin=<ltrb 0 0 0 0>
      outlineSettings {
       bUseBrushTransparency=false
       roundingType=HalfHeightRadius
       width=0
       color {colorUseRule=UseColor_Specified, specifiedColor=<rgba 0 0 0 0>}
       cornerRadii=<0 0 0 0>
      }
      tintColor {colorUseRule=UseColor_Specified, specifiedColor=<rgba 1 1 1 1>}
      uVRegion {bIsValid=false, max=<0 0>, min=<0 0>}
     }
     verticalBottomSlotImage {
      bIsDynamicallyLoaded=false
      drawAs=NoDrawType
      imageType=NoImage
      mirroring=NoMirror
      resourceName=None
      resourceObject=null
      tiling=NoTile
      imageSize=<0 0>
      margin=<ltrb 0 0 0 0>
      outlineSettings {
       bUseBrushTransparency=false
       roundingType=HalfHeightRadius
       width=0
       color {colorUseRule=UseColor_Specified, specifiedColor=<rgba 0 0 0 0>}
       cornerRadii=<0 0 0 0>
      }
      tintColor {colorUseRule=UseColor_Specified, specifiedColor=<rgba 1 1 1 1>}
      uVRegion {bIsValid=false, max=<0 0>, min=<0 0>}
     }
     verticalTopSlotImage {
      bIsDynamicallyLoaded=false
      drawAs=NoDrawType
      imageType=NoImage
      mirroring=NoMirror
      resourceName=None
      resourceObject=null
      tiling=NoTile
      imageSize=<0 0>
      margin=<ltrb 0 0 0 0>
      outlineSettings {
       bUseBrushTransparency=false
       roundingType=HalfHeightRadius
       width=0
       color {colorUseRule=UseColor_Specified, specifiedColor=<rgba 0 0 0 0>}
       cornerRadii=<0 0 0 0>
      }
      tintColor {colorUseRule=UseColor_Specified, specifiedColor=<rgba 1 1 1 1>}
      uVRegion {bIsValid=false, max=<0 0>, min=<0 0>}
     }
    }
   }
   slot {class=VerticalBoxSlot, properties {size {sizeRule=Fill, value=1}}}
  }
  LeftDivider {
   widgetClassPath=/Game/Athena/Editor/MapLayersAutomation/SubWidgets/EUW_ResizeDivider
   bIsVariable=true
   parent=HBMain
   widget=LeftDivider
   properties {visibility=Collapsed}
   slot {class=HorizontalBoxSlot}
  }
  Main {
   widgetClassPath=&VerticalBox
   parent=HBMain
   widget=Main
   slot {class=HorizontalBoxSlot, properties {size {sizeRule=Fill, value=0.5}}}
  }
  WrapBox_0 {
   widgetClassPath=&WrapBox
   parent=Main
   widget=WrapBox_0
   properties {wrapSize=512, innerSlotPadding=<2 2>}
   slot {class=VerticalBoxSlot, properties {padding=<ltrb 4 4 4 0>}}
  }
  btnExpander {
   widgetClassPath=&EditorUtilityButton
   parent=WrapBox_0
   widget=btnExpander
   properties {
    toolTipText=Toggle Folder view
    widgetStyle {
     clickedSlateSound {resourceObject=null}
     disabled {
      bIsDynamicallyLoaded=false
      drawAs=RoundedBox
      imageType=NoImage
      mirroring=NoMirror
      resourceName=None
      resourceObject=null
      tiling=NoTile
      imageSize=<0 0>
      margin=<ltrb 0 0 0 0>
      outlineSettings {
       bUseBrushTransparency=false
       roundingType=FixedRadius
       width=1
       color {colorUseRule=UseColor_ColorTable, specifiedColor=<rgba 0 0 0 0>}
       cornerRadii=<4 4 4 4>
      }
      tintColor {colorUseRule=UseColor_Specified, specifiedColor=<rgba 0.0395 0.0395 0.0395 1>}
      uVRegion {bIsValid=false, max=<0 0>, min=<0 0>}
     }
     disabledForeground {colorUseRule=UseColor_ColorTable, specifiedColor=<rgba 0 0 0 0>}
     hovered {
      bIsDynamicallyLoaded=false
      drawAs=RoundedBox
      imageType=NoImage
      mirroring=NoMirror
      resourceName=None
      resourceObject=null
      tiling=NoTile
      imageSize=<32 32>
      margin=<ltrb 0 0 0 0>
      outlineSettings {
       bUseBrushTransparency=false
       roundingType=FixedRadius
       width=1
       color {colorUseRule=UseColor_ColorTable, specifiedColor=<rgba 0 0 0 0>}
       cornerRadii=<4 4 4 4>
      }
      tintColor {colorUseRule=UseColor_Specified, specifiedColor=<rgba 0.0953 0.0953 0.0953 1>}
      uVRegion {bIsValid=false, max=<0 0>, min=<0 0>}
     }
     hoveredForeground {colorUseRule=UseColor_ColorTable, specifiedColor=<rgba 0 0 0 0>}
     hoveredSlateSound {resourceObject=null}
     normal {
      bIsDynamicallyLoaded=false
      drawAs=RoundedBox
      imageType=NoImage
      mirroring=NoMirror
      resourceName=None
      resourceObject=null
      tiling=NoTile
      imageSize=<32 32>
      margin=<ltrb 0 0 0 0>
      outlineSettings {
       bUseBrushTransparency=false
       roundingType=FixedRadius
       width=1
       color {colorUseRule=UseColor_ColorTable, specifiedColor=<rgba 0 0 0 0>}
       cornerRadii=<4 4 4 4>
      }
      tintColor {colorUseRule=UseColor_Specified, specifiedColor=<rgba 0.0395 0.0395 0.0395 1>}
      uVRegion {bIsValid=false, max=<0 0>, min=<0 0>}
     }
     normalForeground {colorUseRule=UseColor_ColorTable, specifiedColor=<rgba 0 0 0 0>}
     normalPadding=<ltrb 4 4 4 4>
     pressed {
      bIsDynamicallyLoaded=false
      drawAs=RoundedBox
      imageType=NoImage
      mirroring=NoMirror
      resourceName=None
      resourceObject=null
      tiling=NoTile
      imageSize=<32 32>
      margin=<ltrb 0 0 0 0>
      outlineSettings {
       bUseBrushTransparency=false
       roundingType=FixedRadius
       width=1
       color {colorUseRule=UseColor_ColorTable, specifiedColor=<rgba 0 0 0 0>}
       cornerRadii=<4 4 4 4>
      }
      tintColor {colorUseRule=UseColor_Specified, specifiedColor=<rgba 0.0284 0.0284 0.0284 1>}
      uVRegion {bIsValid=false, max=<0 0>, min=<0 0>}
     }
     pressedForeground {colorUseRule=UseColor_ColorTable, specifiedColor=<rgba 0 0 0 0>}
     pressedPadding=<ltrb 4 2.5 4 0.5>
     pressedSlateSound {resourceObject=null}
    }
   }
   slot {class=WrapBoxSlot, properties {horizontalAlignment=HAlign_Center, verticalAlignment=VAlign_Center}}
  }
  img_expander {
   widgetClassPath=&Image
   bIsVariable=true
   parent=btnExpander
   widget=img_expander
   properties {brush {
     bIsDynamicallyLoaded=false
     drawAs=Image
     imageType=FullColor
     mirroring=NoMirror
     resourceName=None
     resourceObject=&anglesrightsolid
     tiling=NoTile
     imageSize=<24 24>
     margin=<ltrb 0 0 0 0>
     outlineSettings {
      bUseBrushTransparency=false
      roundingType=HalfHeightRadius
      width=0
      color {colorUseRule=UseColor_Specified, specifiedColor=<rgba 0 0 0 0>}
      cornerRadii=<0 0 0 0>
     }
     tintColor {colorUseRule=UseColor_Specified, specifiedColor=<rgba 1 1 1 1>}
     uVRegion {bIsValid=false, max=<0 0>, min=<0 0>}
    }}
   slot {class=ButtonSlot, properties {padding=<ltrb 0 0 0 0>}}
  }
  SizeBox_147 {
   widgetClassPath=&SizeBox
   parent=WrapBox_0
   widget=SizeBox_147
   properties {bOverride_MaxDesiredHeight=true, maxDesiredHeight=32}
   slot {class=WrapBoxSlot, properties {horizontalAlignment=HAlign_Center, verticalAlignment=VAlign_Center}}
  }
  SaveActions {
   widgetClassPath=/Game/NeverCook/UI/DropdownButton/DropdownButton
   bIsVariable=true
   parent=SizeBox_147
   widget=SaveActions
   properties {
    mainButtonText=+
    font {
     bForceMonospaced=false
     bMaterialIsStencil=false
     fontMaterial=null
     fontObject=&Roboto
     letterSpacing=0
     monospacedWidth=1
     size=12
     skewAmount=0
     typefaceFontName=Bold
     outlineSettings {
      bApplyOutlineToDropShadows=false
      bMiteredCorners=false
      bSeparateFillAlpha=false
      outlineMaterial=null
      outlineSize=0
      outlineColor=<rgba 0 0 0 1>
     }
    }
    foregroundColor {colorUseRule=UseColor_Specified, specifiedColor=<rgba 0.0035 0.1459 0.3531 1>}
   }
   slot {class=SizeBoxSlot}
  }
  btnUndo {
   widgetClassPath=&EditorUtilityButton
   parent=WrapBox_0
   widget=btnUndo
   properties {
    bIsEnabled=false
    toolTipText="[NOT IMPLEMENTED YET]\r\nUndo latest change"
    widgetStyle {
     clickedSlateSound {resourceObject=null}
     disabled {
      bIsDynamicallyLoaded=false
      drawAs=RoundedBox
      imageType=NoImage
      mirroring=NoMirror
      resourceName=None
      resourceObject=null
      tiling=NoTile
      imageSize=<0 0>
      margin=<ltrb 0 0 0 0>
      outlineSettings {
       bUseBrushTransparency=false
       roundingType=FixedRadius
       width=1
       color {colorUseRule=UseColor_ColorTable, specifiedColor=<rgba 0 0 0 0>}
       cornerRadii=<4 4 4 4>
      }
      tintColor {colorUseRule=UseColor_Specified, specifiedColor=<rgba 0.0395 0.0395 0.0395 1>}
      uVRegion {bIsValid=false, max=<0 0>, min=<0 0>}
     }
     disabledForeground {colorUseRule=UseColor_ColorTable, specifiedColor=<rgba 0 0 0 0>}
     hovered {
      bIsDynamicallyLoaded=false
      drawAs=RoundedBox
      imageType=NoImage
      mirroring=NoMirror
      resourceName=None
      resourceObject=null
      tiling=NoTile
      imageSize=<32 32>
      margin=<ltrb 0 0 0 0>
      outlineSettings {
       bUseBrushTransparency=false
       roundingType=FixedRadius
       width=1
       color {colorUseRule=UseColor_ColorTable, specifiedColor=<rgba 0 0 0 0>}
       cornerRadii=<4 4 4 4>
      }
      tintColor {colorUseRule=UseColor_Specified, specifiedColor=<rgba 0.0953 0.0953 0.0953 1>}
      uVRegion {bIsValid=false, max=<0 0>, min=<0 0>}
     }
     hoveredForeground {colorUseRule=UseColor_ColorTable, specifiedColor=<rgba 0 0 0 0>}
     hoveredSlateSound {resourceObject=null}
     normal {
      bIsDynamicallyLoaded=false
      drawAs=RoundedBox
      imageType=NoImage
      mirroring=NoMirror
      resourceName=None
      resourceObject=null
      tiling=NoTile
      imageSize=<32 32>
      margin=<ltrb 0 0 0 0>
      outlineSettings {
       bUseBrushTransparency=false
       roundingType=FixedRadius
       width=1
       color {colorUseRule=UseColor_ColorTable, specifiedColor=<rgba 0 0 0 0>}
       cornerRadii=<4 4 4 4>
      }
      tintColor {colorUseRule=UseColor_Specified, specifiedColor=<rgba 0.0395 0.0395 0.0395 1>}
      uVRegion {bIsValid=false, max=<0 0>, min=<0 0>}
     }
     normalForeground {colorUseRule=UseColor_ColorTable, specifiedColor=<rgba 0 0 0 0>}
     normalPadding=<ltrb 4 4 4 4>
     pressed {
      bIsDynamicallyLoaded=false
      drawAs=RoundedBox
      imageType=NoImage
      mirroring=NoMirror
      resourceName=None
      resourceObject=null
      tiling=NoTile
      imageSize=<32 32>
      margin=<ltrb 0 0 0 0>
      outlineSettings {
       bUseBrushTransparency=false
       roundingType=FixedRadius
       width=1
       color {colorUseRule=UseColor_ColorTable, specifiedColor=<rgba 0 0 0 0>}
       cornerRadii=<4 4 4 4>
      }
      tintColor {colorUseRule=UseColor_Specified, specifiedColor=<rgba 0.0284 0.0284 0.0284 1>}
      uVRegion {bIsValid=false, max=<0 0>, min=<0 0>}
     }
     pressedForeground {colorUseRule=UseColor_ColorTable, specifiedColor=<rgba 0 0 0 0>}
     pressedPadding=<ltrb 4 2.5 4 0.5>
     pressedSlateSound {resourceObject=null}
    }
   }
   slot {class=WrapBoxSlot, properties {horizontalAlignment=HAlign_Center, verticalAlignment=VAlign_Center}}
  }
  Image_47 {
   widgetClassPath=&Image
   parent=btnUndo
   widget=Image_47
   properties {brush {
     bIsDynamicallyLoaded=false
     drawAs=Image
     imageType=FullColor
     mirroring=NoMirror
     resourceName=None
     resourceObject=/AnimationLibrary/resources/icons/rotateleftsolid
     tiling=NoTile
     imageSize=<24 24>
     margin=<ltrb 0 0 0 0>
     outlineSettings {
      bUseBrushTransparency=false
      roundingType=HalfHeightRadius
      width=0
      color {colorUseRule=UseColor_Specified, specifiedColor=<rgba 0 0 0 0>}
      cornerRadii=<0 0 0 0>
     }
     tintColor {colorUseRule=UseColor_Specified, specifiedColor=<rgba 1 1 1 1>}
     uVRegion {bIsValid=false, max=<0 0>, min=<0 0>}
    }}
   slot {class=ButtonSlot, properties {padding=<ltrb 0 0 0 0>}}
  }
  btnRedo {
   widgetClassPath=&EditorUtilityButton
   parent=WrapBox_0
   widget=btnRedo
   properties {
    bIsEnabled=false
    toolTipText="[NOT IMPLEMENTED YET]\r\nRedo latest change"
    widgetStyle {
     clickedSlateSound {resourceObject=null}
     disabled {
      bIsDynamicallyLoaded=false
      drawAs=RoundedBox
      imageType=NoImage
      mirroring=NoMirror
      resourceName=None
      resourceObject=null
      tiling=NoTile
      imageSize=<0 0>
      margin=<ltrb 0 0 0 0>
      outlineSettings {
       bUseBrushTransparency=false
       roundingType=FixedRadius
       width=1
       color {colorUseRule=UseColor_ColorTable, specifiedColor=<rgba 0 0 0 0>}
       cornerRadii=<4 4 4 4>
      }
      tintColor {colorUseRule=UseColor_Specified, specifiedColor=<rgba 0.0395 0.0395 0.0395 1>}
      uVRegion {bIsValid=false, max=<0 0>, min=<0 0>}
     }
     disabledForeground {colorUseRule=UseColor_ColorTable, specifiedColor=<rgba 0 0 0 0>}
     hovered {
      bIsDynamicallyLoaded=false
      drawAs=RoundedBox
      imageType=NoImage
      mirroring=NoMirror
      resourceName=None
      resourceObject=null
      tiling=NoTile
      imageSize=<32 32>
      margin=<ltrb 0 0 0 0>
      outlineSettings {
       bUseBrushTransparency=false
       roundingType=FixedRadius
       width=1
       color {colorUseRule=UseColor_ColorTable, specifiedColor=<rgba 0 0 0 0>}
       cornerRadii=<4 4 4 4>
      }
      tintColor {colorUseRule=UseColor_Specified, specifiedColor=<rgba 0.0953 0.0953 0.0953 1>}
      uVRegion {bIsValid=false, max=<0 0>, min=<0 0>}
     }
     hoveredForeground {colorUseRule=UseColor_ColorTable, specifiedColor=<rgba 0 0 0 0>}
     hoveredSlateSound {resourceObject=null}
     normal {
      bIsDynamicallyLoaded=false
      drawAs=RoundedBox
      imageType=NoImage
      mirroring=NoMirror
      resourceName=None
      resourceObject=null
      tiling=NoTile
      imageSize=<32 32>
      margin=<ltrb 0 0 0 0>
      outlineSettings {
       bUseBrushTransparency=false
       roundingType=FixedRadius
       width=1
       color {colorUseRule=UseColor_ColorTable, specifiedColor=<rgba 0 0 0 0>}
       cornerRadii=<4 4 4 4>
      }
      tintColor {colorUseRule=UseColor_Specified, specifiedColor=<rgba 0.0395 0.0395 0.0395 1>}
      uVRegion {bIsValid=false, max=<0 0>, min=<0 0>}
     }
     normalForeground {colorUseRule=UseColor_ColorTable, specifiedColor=<rgba 0 0 0 0>}
     normalPadding=<ltrb 4 4 4 4>
     pressed {
      bIsDynamicallyLoaded=false
      drawAs=RoundedBox
      imageType=NoImage
      mirroring=NoMirror
      resourceName=None
      resourceObject=null
      tiling=NoTile
      imageSize=<32 32>
      margin=<ltrb 0 0 0 0>
      outlineSettings {
       bUseBrushTransparency=false
       roundingType=FixedRadius
       width=1
       color {colorUseRule=UseColor_ColorTable, specifiedColor=<rgba 0 0 0 0>}
       cornerRadii=<4 4 4 4>
      }
      tintColor {colorUseRule=UseColor_Specified, specifiedColor=<rgba 0.0284 0.0284 0.0284 1>}
      uVRegion {bIsValid=false, max=<0 0>, min=<0 0>}
     }
     pressedForeground {colorUseRule=UseColor_ColorTable, specifiedColor=<rgba 0 0 0 0>}
     pressedPadding=<ltrb 4 2.5 4 0.5>
     pressedSlateSound {resourceObject=null}
    }
   }
   slot {class=WrapBoxSlot, properties {horizontalAlignment=HAlign_Center, verticalAlignment=VAlign_Center}}
  }
  Image_106 {
   widgetClassPath=&Image
   parent=btnRedo
   widget=Image_106
   properties {brush {
     bIsDynamicallyLoaded=false
     drawAs=Image
     imageType=FullColor
     mirroring=NoMirror
     resourceName=None
     resourceObject=/AnimationLibrary/resources/icons/rotaterightsolid
     tiling=NoTile
     imageSize=<24 24>
     margin=<ltrb 0 0 0 0>
     outlineSettings {
      bUseBrushTransparency=false
      roundingType=HalfHeightRadius
      width=0
      color {colorUseRule=UseColor_Specified, specifiedColor=<rgba 0 0 0 0>}
      cornerRadii=<0 0 0 0>
     }
     tintColor {colorUseRule=UseColor_Specified, specifiedColor=<rgba 1 1 1 1>}
     uVRegion {bIsValid=false, max=<0 0>, min=<0 0>}
    }}
   slot {class=ButtonSlot, properties {padding=<ltrb 0 0 0 0>}}
  }
  btnSaveAll {
   widgetClassPath=&EditorUtilityButton
   parent=WrapBox_0
   widget=btnSaveAll
   properties {
    toolTipText=Save all assets changes
    widgetStyle {
     clickedSlateSound {resourceObject=null}
     disabled {
      bIsDynamicallyLoaded=false
      drawAs=RoundedBox
      imageType=NoImage
      mirroring=NoMirror
      resourceName=None
      resourceObject=null
      tiling=NoTile
      imageSize=<0 0>
      margin=<ltrb 0 0 0 0>
      outlineSettings {
       bUseBrushTransparency=false
       roundingType=FixedRadius
       width=1
       color {colorUseRule=UseColor_ColorTable, specifiedColor=<rgba 0 0 0 0>}
       cornerRadii=<4 4 4 4>
      }
      tintColor {colorUseRule=UseColor_Specified, specifiedColor=<rgba 0.0395 0.0395 0.0395 1>}
      uVRegion {bIsValid=false, max=<0 0>, min=<0 0>}
     }
     disabledForeground {colorUseRule=UseColor_ColorTable, specifiedColor=<rgba 0 0 0 0>}
     hovered {
      bIsDynamicallyLoaded=false
      drawAs=RoundedBox
      imageType=NoImage
      mirroring=NoMirror
      resourceName=None
      resourceObject=null
      tiling=NoTile
      imageSize=<32 32>
      margin=<ltrb 0 0 0 0>
      outlineSettings {
       bUseBrushTransparency=false
       roundingType=FixedRadius
       width=1
       color {colorUseRule=UseColor_ColorTable, specifiedColor=<rgba 0 0 0 0>}
       cornerRadii=<4 4 4 4>
      }
      tintColor {colorUseRule=UseColor_Specified, specifiedColor=<rgba 0.0953 0.0953 0.0953 1>}
      uVRegion {bIsValid=false, max=<0 0>, min=<0 0>}
     }
     hoveredForeground {colorUseRule=UseColor_ColorTable, specifiedColor=<rgba 0 0 0 0>}
     hoveredSlateSound {resourceObject=null}
     normal {
      bIsDynamicallyLoaded=false
      drawAs=RoundedBox
      imageType=NoImage
      mirroring=NoMirror
      resourceName=None
      resourceObject=null
      tiling=NoTile
      imageSize=<32 32>
      margin=<ltrb 0 0 0 0>
      outlineSettings {
       bUseBrushTransparency=false
       roundingType=FixedRadius
       width=1
       color {colorUseRule=UseColor_ColorTable, specifiedColor=<rgba 0 0 0 0>}
       cornerRadii=<4 4 4 4>
      }
      tintColor {colorUseRule=UseColor_Specified, specifiedColor=<rgba 0.0395 0.0395 0.0395 1>}
      uVRegion {bIsValid=false, max=<0 0>, min=<0 0>}
     }
     normalForeground {colorUseRule=UseColor_ColorTable, specifiedColor=<rgba 0 0 0 0>}
     normalPadding=<ltrb 4 4 4 4>
     pressed {
      bIsDynamicallyLoaded=false
      drawAs=RoundedBox
      imageType=NoImage
      mirroring=NoMirror
      resourceName=None
      resourceObject=null
      tiling=NoTile
      imageSize=<32 32>
      margin=<ltrb 0 0 0 0>
      outlineSettings {
       bUseBrushTransparency=false
       roundingType=FixedRadius
       width=1
       color {colorUseRule=UseColor_ColorTable, specifiedColor=<rgba 0 0 0 0>}
       cornerRadii=<4 4 4 4>
      }
      tintColor {colorUseRule=UseColor_Specified, specifiedColor=<rgba 0.0284 0.0284 0.0284 1>}
      uVRegion {bIsValid=false, max=<0 0>, min=<0 0>}
     }
     pressedForeground {colorUseRule=UseColor_ColorTable, specifiedColor=<rgba 0 0 0 0>}
     pressedPadding=<ltrb 4 2.5 4 0.5>
     pressedSlateSound {resourceObject=null}
    }
   }
   slot {class=WrapBoxSlot, properties {horizontalAlignment=HAlign_Center, verticalAlignment=VAlign_Center}}
  }
  Image_1 {
   widgetClassPath=&Image
   parent=btnSaveAll
   widget=Image_1
   properties {brush {
     bIsDynamicallyLoaded=false
     drawAs=Image
     imageType=FullColor
     mirroring=NoMirror
     resourceName=None
     resourceObject=/AnimationLibrary/resources/icons/floppydisksolid
     tiling=NoTile
     imageSize=<24 24>
     margin=<ltrb 0 0 0 0>
     outlineSettings {
      bUseBrushTransparency=false
      roundingType=HalfHeightRadius
      width=0
      color {colorUseRule=UseColor_Specified, specifiedColor=<rgba 0 0 0 0>}
      cornerRadii=<0 0 0 0>
     }
     tintColor {colorUseRule=UseColor_Specified, specifiedColor=<rgba 0 0.3913 1 1>}
     uVRegion {bIsValid=false, max=<0 0>, min=<0 0>}
    }}
   slot {class=ButtonSlot, properties {padding=<ltrb 0 0 0 0>}}
  }
  btnRefresh {
   widgetClassPath=&EditorUtilityButton
   parent=WrapBox_0
   widget=btnRefresh
   properties {
    toolTipText=Refresh Assets from Disk
    widgetStyle {
     clickedSlateSound {resourceObject=null}
     disabled {
      bIsDynamicallyLoaded=false
      drawAs=RoundedBox
      imageType=NoImage
      mirroring=NoMirror
      resourceName=None
      resourceObject=null
      tiling=NoTile
      imageSize=<0 0>
      margin=<ltrb 0 0 0 0>
      outlineSettings {
       bUseBrushTransparency=false
       roundingType=FixedRadius
       width=1
       color {colorUseRule=UseColor_ColorTable, specifiedColor=<rgba 0 0 0 0>}
       cornerRadii=<4 4 4 4>
      }
      tintColor {colorUseRule=UseColor_Specified, specifiedColor=<rgba 0.0395 0.0395 0.0395 1>}
      uVRegion {bIsValid=false, max=<0 0>, min=<0 0>}
     }
     disabledForeground {colorUseRule=UseColor_ColorTable, specifiedColor=<rgba 0 0 0 0>}
     hovered {
      bIsDynamicallyLoaded=false
      drawAs=RoundedBox
      imageType=NoImage
      mirroring=NoMirror
      resourceName=None
      resourceObject=null
      tiling=NoTile
      imageSize=<32 32>
      margin=<ltrb 0 0 0 0>
      outlineSettings {
       bUseBrushTransparency=false
       roundingType=FixedRadius
       width=1
       color {colorUseRule=UseColor_ColorTable, specifiedColor=<rgba 0 0 0 0>}
       cornerRadii=<4 4 4 4>
      }
      tintColor {colorUseRule=UseColor_Specified, specifiedColor=<rgba 0.0953 0.0953 0.0953 1>}
      uVRegion {bIsValid=false, max=<0 0>, min=<0 0>}
     }
     hoveredForeground {colorUseRule=UseColor_ColorTable, specifiedColor=<rgba 0 0 0 0>}
     hoveredSlateSound {resourceObject=null}
     normal {
      bIsDynamicallyLoaded=false
      drawAs=RoundedBox
      imageType=NoImage
      mirroring=NoMirror
      resourceName=None
      resourceObject=null
      tiling=NoTile
      imageSize=<32 32>
      margin=<ltrb 0 0 0 0>
      outlineSettings {
       bUseBrushTransparency=false
       roundingType=FixedRadius
       width=1
       color {colorUseRule=UseColor_ColorTable, specifiedColor=<rgba 0 0 0 0>}
       cornerRadii=<4 4 4 4>
      }
      tintColor {colorUseRule=UseColor_Specified, specifiedColor=<rgba 0.0395 0.0395 0.0395 1>}
      uVRegion {bIsValid=false, max=<0 0>, min=<0 0>}
     }
     normalForeground {colorUseRule=UseColor_ColorTable, specifiedColor=<rgba 0 0 0 0>}
     normalPadding=<ltrb 4 4 4 4>
     pressed {
      bIsDynamicallyLoaded=false
      drawAs=RoundedBox
      imageType=NoImage
      mirroring=NoMirror
      resourceName=None
      resourceObject=null
      tiling=NoTile
      imageSize=<32 32>
      margin=<ltrb 0 0 0 0>
      outlineSettings {
       bUseBrushTransparency=false
       roundingType=FixedRadius
       width=1
       color {colorUseRule=UseColor_ColorTable, specifiedColor=<rgba 0 0 0 0>}
       cornerRadii=<4 4 4 4>
      }
      tintColor {colorUseRule=UseColor_Specified, specifiedColor=<rgba 0.0284 0.0284 0.0284 1>}
      uVRegion {bIsValid=false, max=<0 0>, min=<0 0>}
     }
     pressedForeground {colorUseRule=UseColor_ColorTable, specifiedColor=<rgba 0 0 0 0>}
     pressedPadding=<ltrb 4 2.5 4 0.5>
     pressedSlateSound {resourceObject=null}
    }
   }
   slot {class=WrapBoxSlot, properties {horizontalAlignment=HAlign_Center, verticalAlignment=VAlign_Center}}
  }
  Image_3 {
   widgetClassPath=&Image
   parent=btnRefresh
   widget=Image_3
   properties {brush {
     bIsDynamicallyLoaded=false
     drawAs=Image
     imageType=FullColor
     mirroring=NoMirror
     resourceName=None
     resourceObject=/AnimationLibrary/resources/icons/arrowsrotatesolid
     tiling=NoTile
     imageSize=<24 24>
     margin=<ltrb 0 0 0 0>
     outlineSettings {
      bUseBrushTransparency=false
      roundingType=HalfHeightRadius
      width=0
      color {colorUseRule=UseColor_Specified, specifiedColor=<rgba 0 0 0 0>}
      cornerRadii=<0 0 0 0>
     }
     tintColor {colorUseRule=UseColor_Specified, specifiedColor=<rgba 0 0.5938 0.2375 1>}
     uVRegion {bIsValid=false, max=<0 0>, min=<0 0>}
    }}
   slot {class=ButtonSlot, properties {padding=<ltrb 0 0 0 0>}}
  }
  btnSettings {
   widgetClassPath=&EditorUtilityButton
   parent=WrapBox_0
   widget=btnSettings
   properties {
    toolTipText=Open Library Settings
    widgetStyle {
     clickedSlateSound {resourceObject=null}
     disabled {
      bIsDynamicallyLoaded=false
      drawAs=RoundedBox
      imageType=NoImage
      mirroring=NoMirror
      resourceName=None
      resourceObject=null
      tiling=NoTile
      imageSize=<0 0>
      margin=<ltrb 0 0 0 0>
      outlineSettings {
       bUseBrushTransparency=false
       roundingType=FixedRadius
       width=1
       color {colorUseRule=UseColor_ColorTable, specifiedColor=<rgba 0 0 0 0>}
       cornerRadii=<4 4 4 4>
      }
      tintColor {colorUseRule=UseColor_Specified, specifiedColor=<rgba 0.0395 0.0395 0.0395 1>}
      uVRegion {bIsValid=false, max=<0 0>, min=<0 0>}
     }
     disabledForeground {colorUseRule=UseColor_ColorTable, specifiedColor=<rgba 0 0 0 0>}
     hovered {
      bIsDynamicallyLoaded=false
      drawAs=RoundedBox
      imageType=NoImage
      mirroring=NoMirror
      resourceName=None
      resourceObject=null
      tiling=NoTile
      imageSize=<32 32>
      margin=<ltrb 0 0 0 0>
      outlineSettings {
       bUseBrushTransparency=false
       roundingType=FixedRadius
       width=1
       color {colorUseRule=UseColor_ColorTable, specifiedColor=<rgba 0 0 0 0>}
       cornerRadii=<4 4 4 4>
      }
      tintColor {colorUseRule=UseColor_Specified, specifiedColor=<rgba 0.0953 0.0953 0.0953 1>}
      uVRegion {bIsValid=false, max=<0 0>, min=<0 0>}
     }
     hoveredForeground {colorUseRule=UseColor_ColorTable, specifiedColor=<rgba 0 0 0 0>}
     hoveredSlateSound {resourceObject=null}
     normal {
      bIsDynamicallyLoaded=false
      drawAs=RoundedBox
      imageType=NoImage
      mirroring=NoMirror
      resourceName=None
      resourceObject=null
      tiling=NoTile
      imageSize=<32 32>
      margin=<ltrb 0 0 0 0>
      outlineSettings {
       bUseBrushTransparency=false
       roundingType=FixedRadius
       width=1
       color {colorUseRule=UseColor_ColorTable, specifiedColor=<rgba 0 0 0 0>}
       cornerRadii=<4 4 4 4>
      }
      tintColor {colorUseRule=UseColor_Specified, specifiedColor=<rgba 0.0395 0.0395 0.0395 1>}
      uVRegion {bIsValid=false, max=<0 0>, min=<0 0>}
     }
     normalForeground {colorUseRule=UseColor_ColorTable, specifiedColor=<rgba 0 0 0 0>}
     normalPadding=<ltrb 4 4 4 4>
     pressed {
      bIsDynamicallyLoaded=false
      drawAs=RoundedBox
      imageType=NoImage
      mirroring=NoMirror
      resourceName=None
      resourceObject=null
      tiling=NoTile
      imageSize=<32 32>
      margin=<ltrb 0 0 0 0>
      outlineSettings {
       bUseBrushTransparency=false
       roundingType=FixedRadius
       width=1
       color {colorUseRule=UseColor_ColorTable, specifiedColor=<rgba 0 0 0 0>}
       cornerRadii=<4 4 4 4>
      }
      tintColor {colorUseRule=UseColor_Specified, specifiedColor=<rgba 0.0284 0.0284 0.0284 1>}
      uVRegion {bIsValid=false, max=<0 0>, min=<0 0>}
     }
     pressedForeground {colorUseRule=UseColor_ColorTable, specifiedColor=<rgba 0 0 0 0>}
     pressedPadding=<ltrb 4 2.5 4 0.5>
     pressedSlateSound {resourceObject=null}
    }
   }
   slot {class=WrapBoxSlot, properties {horizontalAlignment=HAlign_Center, verticalAlignment=VAlign_Center}}
  }
  Image_2 {
   widgetClassPath=&Image
   parent=btnSettings
   widget=Image_2
   properties {brush {
     bIsDynamicallyLoaded=false
     drawAs=Image
     imageType=FullColor
     mirroring=NoMirror
     resourceName=None
     resourceObject=&sliderssolid
     tiling=NoTile
     imageSize=<24 24>
     margin=<ltrb 0 0 0 0>
     outlineSettings {
      bUseBrushTransparency=false
      roundingType=HalfHeightRadius
      width=0
      color {colorUseRule=UseColor_Specified, specifiedColor=<rgba 0 0 0 0>}
      cornerRadii=<0 0 0 0>
     }
     tintColor {colorUseRule=UseColor_Specified, specifiedColor=<rgba 0.8385 0.2625 0 1>}
     uVRegion {bIsValid=false, max=<0 0>, min=<0 0>}
    }}
   slot {class=ButtonSlot, properties {padding=<ltrb 0 0 0 0>}}
  }
  btnHelp {
   widgetClassPath=&EditorUtilityButton
   parent=WrapBox_0
   widget=btnHelp
   properties {
    toolTipText=Refresh Assets from Disk
    widgetStyle {
     clickedSlateSound {resourceObject=null}
     disabled {
      bIsDynamicallyLoaded=false
      drawAs=RoundedBox
      imageType=NoImage
      mirroring=NoMirror
      resourceName=None
      resourceObject=null
      tiling=NoTile
      imageSize=<0 0>
      margin=<ltrb 0 0 0 0>
      outlineSettings {
       bUseBrushTransparency=false
       roundingType=FixedRadius
       width=1
       color {colorUseRule=UseColor_ColorTable, specifiedColor=<rgba 0 0 0 0>}
       cornerRadii=<4 4 4 4>
      }
      tintColor {colorUseRule=UseColor_Specified, specifiedColor=<rgba 0.0395 0.0395 0.0395 1>}
      uVRegion {bIsValid=false, max=<0 0>, min=<0 0>}
     }
     disabledForeground {colorUseRule=UseColor_ColorTable, specifiedColor=<rgba 0 0 0 0>}
     hovered {
      bIsDynamicallyLoaded=false
      drawAs=RoundedBox
      imageType=NoImage
      mirroring=NoMirror
      resourceName=None
      resourceObject=null
      tiling=NoTile
      imageSize=<32 32>
      margin=<ltrb 0 0 0 0>
      outlineSettings {
       bUseBrushTransparency=false
       roundingType=FixedRadius
       width=1
       color {colorUseRule=UseColor_ColorTable, specifiedColor=<rgba 0 0 0 0>}
       cornerRadii=<4 4 4 4>
      }
      tintColor {colorUseRule=UseColor_Specified, specifiedColor=<rgba 0.0953 0.0953 0.0953 1>}
      uVRegion {bIsValid=false, max=<0 0>, min=<0 0>}
     }
     hoveredForeground {colorUseRule=UseColor_ColorTable, specifiedColor=<rgba 0 0 0 0>}
     hoveredSlateSound {resourceObject=null}
     normal {
      bIsDynamicallyLoaded=false
      drawAs=RoundedBox
      imageType=NoImage
      mirroring=NoMirror
      resourceName=None
      resourceObject=null
      tiling=NoTile
      imageSize=<32 32>
      margin=<ltrb 0 0 0 0>
      outlineSettings {
       bUseBrushTransparency=false
       roundingType=FixedRadius
       width=1
       color {colorUseRule=UseColor_ColorTable, specifiedColor=<rgba 0 0 0 0>}
       cornerRadii=<4 4 4 4>
      }
      tintColor {colorUseRule=UseColor_Specified, specifiedColor=<rgba 0.0395 0.0395 0.0395 1>}
      uVRegion {bIsValid=false, max=<0 0>, min=<0 0>}
     }
     normalForeground {colorUseRule=UseColor_ColorTable, specifiedColor=<rgba 0 0 0 0>}
     normalPadding=<ltrb 4 4 4 4>
     pressed {
      bIsDynamicallyLoaded=false
      drawAs=RoundedBox
      imageType=NoImage
      mirroring=NoMirror
      resourceName=None
      resourceObject=null
      tiling=NoTile
      imageSize=<32 32>
      margin=<ltrb 0 0 0 0>
      outlineSettings {
       bUseBrushTransparency=false
       roundingType=FixedRadius
       width=1
       color {colorUseRule=UseColor_ColorTable, specifiedColor=<rgba 0 0 0 0>}
       cornerRadii=<4 4 4 4>
      }
      tintColor {colorUseRule=UseColor_Specified, specifiedColor=<rgba 0.0284 0.0284 0.0284 1>}
      uVRegion {bIsValid=false, max=<0 0>, min=<0 0>}
     }
     pressedForeground {colorUseRule=UseColor_ColorTable, specifiedColor=<rgba 0 0 0 0>}
     pressedPadding=<ltrb 4 2.5 4 0.5>
     pressedSlateSound {resourceObject=null}
    }
   }
   slot {class=WrapBoxSlot, properties {horizontalAlignment=HAlign_Center, verticalAlignment=VAlign_Center}}
  }
  Image_4 {
   widgetClassPath=&Image
   parent=btnHelp
   widget=Image_4
   properties {
    toolTipText="Help !"
    brush {
     bIsDynamicallyLoaded=false
     drawAs=Image
     imageType=FullColor
     mirroring=NoMirror
     resourceName=None
     resourceObject=/AnimationLibrary/resources/icons/Slack_icon
     tiling=NoTile
     imageSize=<24 24>
     margin=<ltrb 0 0 0 0>
     outlineSettings {
      bUseBrushTransparency=false
      roundingType=HalfHeightRadius
      width=0
      color {colorUseRule=UseColor_Specified, specifiedColor=<rgba 0 0 0 0>}
      cornerRadii=<0 0 0 0>
     }
     tintColor {colorUseRule=UseColor_Specified, specifiedColor=<rgba 1 1 1 1>}
     uVRegion {bIsValid=false, max=<0 0>, min=<0 0>}
    }
   }
   slot {class=ButtonSlot, properties {padding=<ltrb 0 0 0 0>}}
  }
  HorizontalBox_0 {
   widgetClassPath=&HorizontalBox
   parent=WrapBox_0
   widget=HorizontalBox_0
   slot {class=WrapBoxSlot, properties {horizontalAlignment=HAlign_Right, verticalAlignment=VAlign_Center}}
  }
  chkFilterAnim {
   widgetClassPath=&EditorUtilityCheckBox
   bIsVariable=true
   parent=HorizontalBox_0
   widget=chkFilterAnim
   properties {
    checkedState=Checked
    toolTipText=Show Anim Assets
    widgetStyle {
     checkBoxType=ToggleButton
     backgroundHoveredImage {
      bIsDynamicallyLoaded=false
      drawAs=RoundedBox
      imageType=NoImage
      mirroring=NoMirror
      resourceName=None
      resourceObject=null
      tiling=NoTile
      imageSize=<24 24>
      margin=<ltrb 0 0 0 0>
      outlineSettings {
       bUseBrushTransparency=false
       roundingType=FixedRadius
       width=1
       color {colorUseRule=UseColor_ColorTable, specifiedColor=<rgba 0 0 0 0>}
       cornerRadii=<3 3 3 3>
      }
      tintColor {colorUseRule=UseColor_Specified, specifiedColor=<rgba 0.0048 0.0048 0.0048 1>}
      uVRegion {bIsValid=false, max=<0 0>, min=<0 0>}
     }
     backgroundImage {
      bIsDynamicallyLoaded=false
      drawAs=RoundedBox
      imageType=NoImage
      mirroring=NoMirror
      resourceName=None
      resourceObject=null
      tiling=NoTile
      imageSize=<24 24>
      margin=<ltrb 0 0 0 0>
      outlineSettings {
       bUseBrushTransparency=false
       roundingType=FixedRadius
       width=1
       color {colorUseRule=UseColor_ColorTable, specifiedColor=<rgba 0 0 0 0>}
       cornerRadii=<3 3 3 3>
      }
      tintColor {colorUseRule=UseColor_Specified, specifiedColor=<rgba 0.0048 0.0048 0.0048 1>}
      uVRegion {bIsValid=false, max=<0 0>, min=<0 0>}
     }
     backgroundPressedImage {
      bIsDynamicallyLoaded=false
      drawAs=RoundedBox
      imageType=NoImage
      mirroring=NoMirror
      resourceName=None
      resourceObject=null
      tiling=NoTile
      imageSize=<24 24>
      margin=<ltrb 0 0 0 0>
      outlineSettings {
       bUseBrushTransparency=false
       roundingType=FixedRadius
       width=1
       color {colorUseRule=UseColor_ColorTable, specifiedColor=<rgba 0 0 0 0>}
       cornerRadii=<3 3 3 3>
      }
      tintColor {colorUseRule=UseColor_Specified, specifiedColor=<rgba 0.0048 0.0048 0.0048 1>}
      uVRegion {bIsValid=false, max=<0 0>, min=<0 0>}
     }
     borderBackgroundColor {colorUseRule=UseColor_Specified, specifiedColor=<rgba 1 1 1 1>}
     checkedForeground {colorUseRule=UseColor_Specified, specifiedColor=<rgba 1 1 1 1>}
     checkedHoveredForeground {colorUseRule=UseColor_Specified, specifiedColor=<rgba 1 1 1 1>}
     checkedHoveredImage {
      bIsDynamicallyLoaded=false
      drawAs=RoundedBox
      imageType=FullColor
      mirroring=NoMirror
      resourceName=None
      resourceObject=/AnimationLibrary/resources/icons/checkedsolid
      tiling=NoTile
      imageSize=<128 128>
      margin=<ltrb 0 0 0 0>
      outlineSettings {
       bUseBrushTransparency=true
       roundingType=FixedRadius
       width=1
       color {colorUseRule=UseColor_Specified, specifiedColor=<rgba 0 1 1 1>}
       cornerRadii=<4 4 4 4>
      }
      tintColor {colorUseRule=UseColor_Specified, specifiedColor=<rgba 0 1 1 1>}
      uVRegion {bIsValid=false, max=<0 0>, min=<0 0>}
     }
     checkedImage {
      bIsDynamicallyLoaded=false
      drawAs=RoundedBox
      imageType=FullColor
      mirroring=NoMirror
      resourceName=None
      resourceObject=&checkedsolid
      tiling=NoTile
      imageSize=<24 24>
      margin=<ltrb 0 0 0 0>
      outlineSettings {
       bUseBrushTransparency=true
       roundingType=FixedRadius
       width=1
       color {colorUseRule=UseColor_Specified, specifiedColor=<rgba 0 0.5 0.5 1>}
       cornerRadii=<4 4 4 4>
      }
      tintColor {colorUseRule=UseColor_Specified, specifiedColor=<rgba 0 0.5 0.5 1>}
      uVRegion {bIsValid=false, max=<0 0>, min=<0 0>}
     }
     checkedPressedForeground {colorUseRule=UseColor_Specified, specifiedColor=<rgba 1 1 1 1>}
     checkedPressedImage {
      bIsDynamicallyLoaded=false
      drawAs=RoundedBox
      imageType=FullColor
      mirroring=NoMirror
      resourceName=None
      resourceObject=&checkedsolid
      tiling=NoTile
      imageSize=<128 128>
      margin=<ltrb 0 0 0 0>
      outlineSettings {
       bUseBrushTransparency=true
       roundingType=FixedRadius
       width=1
       color {colorUseRule=UseColor_Specified, specifiedColor=<rgba 0 0.2 0.2 1>}
       cornerRadii=<4 4 4 4>
      }
      tintColor {colorUseRule=UseColor_Specified, specifiedColor=<rgba 0 0.2 0.2 1>}
      uVRegion {bIsValid=false, max=<0 0>, min=<0 0>}
     }
     checkedSlateSound {resourceObject=null}
     foregroundColor {colorUseRule=UseColor_Specified, specifiedColor=<rgba 1 1 1 1>}
     hoveredForeground {colorUseRule=UseColor_Specified, specifiedColor=<rgba 1 1 1 1>}
     hoveredSlateSound {resourceObject=null}
     padding=<ltrb 14 14 14 14>
     pressedForeground {colorUseRule=UseColor_Specified, specifiedColor=<rgba 1 1 1 1>}
     uncheckedHoveredImage {
      bIsDynamicallyLoaded=false
      drawAs=RoundedBox
      imageType=NoImage
      mirroring=NoMirror
      resourceName=None
      resourceObject=null
      tiling=NoTile
      imageSize=<24 24>
      margin=<ltrb 0 0 0 0>
      outlineSettings {
       bUseBrushTransparency=false
       roundingType=FixedRadius
       width=1
       color {colorUseRule=UseColor_Specified, specifiedColor=<rgba 0 1 1 1>}
       cornerRadii=<4 4 4 4>
      }
      tintColor {colorUseRule=UseColor_Specified, specifiedColor=<rgba 0.2 0.1 0 0>}
      uVRegion {bIsValid=false, max=<0 0>, min=<0 0>}
     }
     uncheckedImage {
      bIsDynamicallyLoaded=false
      drawAs=RoundedBox
      imageType=NoImage
      mirroring=NoMirror
      resourceName=None
      resourceObject=null
      tiling=NoTile
      imageSize=<24 24>
      margin=<ltrb 0 0 0 0>
      outlineSettings {
       bUseBrushTransparency=false
       roundingType=FixedRadius
       width=1
       color {colorUseRule=UseColor_Specified, specifiedColor=<rgba 0 0.5 0.5 1>}
       cornerRadii=<4 4 4 4>
      }
      tintColor {colorUseRule=UseColor_Specified, specifiedColor=<rgba 0.2 0.1 0 0>}
      uVRegion {bIsValid=false, max=<0 0>, min=<0 0>}
     }
     uncheckedPressedImage {
      bIsDynamicallyLoaded=false
      drawAs=RoundedBox
      imageType=NoImage
      mirroring=NoMirror
      resourceName=None
      resourceObject=null
      tiling=NoTile
      imageSize=<24 24>
      margin=<ltrb 0 0 0 0>
      outlineSettings {
       bUseBrushTransparency=false
       roundingType=FixedRadius
       width=1
       color {colorUseRule=UseColor_Specified, specifiedColor=<rgba 0 0.2 0.2 1>}
       cornerRadii=<4 4 4 4>
      }
      tintColor {colorUseRule=UseColor_Specified, specifiedColor=<rgba 0.2 0.1 0 0>}
      uVRegion {bIsValid=false, max=<0 0>, min=<0 0>}
     }
     uncheckedSlateSound {resourceObject=null}
     undeterminedForeground {colorUseRule=UseColor_Specified, specifiedColor=<rgba 1 1 1 1>}
     undeterminedHoveredImage {
      bIsDynamicallyLoaded=false
      drawAs=Image
      imageType=Vector
      mirroring=NoMirror
      resourceName=../../../Engine/Content/Slate/Starship/CoreWidgets/CheckBox/indeterminate.svg
      resourceObject=null
      tiling=NoTile
      imageSize=<24 24>
      margin=<ltrb 0 0 0 0>
      outlineSettings {
       bUseBrushTransparency=false
       roundingType=HalfHeightRadius
       width=0
       color {colorUseRule=UseColor_Specified, specifiedColor=<rgba 0 0 0 0>}
       cornerRadii=<0 0 0 0>
      }
      tintColor {colorUseRule=UseColor_Specified, specifiedColor=<rgba 0.0044 0.2384 1 1>}
      uVRegion {bIsValid=false, max=<0 0>, min=<0 0>}
     }
     undeterminedImage {
      bIsDynamicallyLoaded=false
      drawAs=Image
      imageType=Vector
      mirroring=NoMirror
      resourceName=../../../Engine/Content/Slate/Starship/CoreWidgets/CheckBox/indeterminate.svg
      resourceObject=null
      tiling=NoTile
      imageSize=<24 24>
      margin=<ltrb 0 0 0 0>
      outlineSettings {
       bUseBrushTransparency=false
       roundingType=HalfHeightRadius
       width=0
       color {colorUseRule=UseColor_Specified, specifiedColor=<rgba 0 0 0 0>}
       cornerRadii=<0 0 0 0>
      }
      tintColor {colorUseRule=UseColor_Specified, specifiedColor=<rgba 0 0.162 0.7454 1>}
      uVRegion {bIsValid=false, max=<0 0>, min=<0 0>}
     }
     undeterminedPressedImage {
      bIsDynamicallyLoaded=false
      drawAs=Image
      imageType=Vector
      mirroring=NoMirror
      resourceName=../../../Engine/Content/Slate/Starship/CoreWidgets/CheckBox/indeterminate.svg
      resourceObject=null
      tiling=NoTile
      imageSize=<24 24>
      margin=<ltrb 0 0 0 0>
      outlineSettings {
       bUseBrushTransparency=false
       roundingType=HalfHeightRadius
       width=0
       color {colorUseRule=UseColor_Specified, specifiedColor=<rgba 0 0 0 0>}
       cornerRadii=<0 0 0 0>
      }
      tintColor {colorUseRule=UseColor_Specified, specifiedColor=<rgba 0 0.0802 0.3515 1>}
      uVRegion {bIsValid=false, max=<0 0>, min=<0 0>}
     }
    }
   }
   slot {class=HorizontalBoxSlot, properties {padding=<ltrb 4 4 4 4>, size {sizeRule=Fill, value=1}}}
  }
  chkFilterPose {
   widgetClassPath=&EditorUtilityCheckBox
   bIsVariable=true
   parent=HorizontalBox_0
   widget=chkFilterPose
   properties {
    checkedState=Checked
    horizontalAlignment=HAlign_Center
    toolTipText=Show Pose Assets
    widgetStyle {
     checkBoxType=ToggleButton
     backgroundHoveredImage {
      bIsDynamicallyLoaded=false
      drawAs=RoundedBox
      imageType=NoImage
      mirroring=NoMirror
      resourceName=None
      resourceObject=null
      tiling=NoTile
      imageSize=<18 18>
      margin=<ltrb 0 0 0 0>
      outlineSettings {
       bUseBrushTransparency=false
       roundingType=FixedRadius
       width=1
       color {colorUseRule=UseColor_ColorTable, specifiedColor=<rgba 0 0 0 0>}
       cornerRadii=<3 3 3 3>
      }
      tintColor {colorUseRule=UseColor_Specified, specifiedColor=<rgba 0.0048 0.0048 0.0048 1>}
      uVRegion {bIsValid=false, max=<0 0>, min=<0 0>}
     }
     backgroundImage {
      bIsDynamicallyLoaded=false
      drawAs=RoundedBox
      imageType=NoImage
      mirroring=NoMirror
      resourceName=None
      resourceObject=null
      tiling=NoTile
      imageSize=<18 18>
      margin=<ltrb 0 0 0 0>
      outlineSettings {
       bUseBrushTransparency=false
       roundingType=FixedRadius
       width=1
       color {colorUseRule=UseColor_ColorTable, specifiedColor=<rgba 0 0 0 0>}
       cornerRadii=<3 3 3 3>
      }
      tintColor {colorUseRule=UseColor_Specified, specifiedColor=<rgba 0.0048 0.0048 0.0048 1>}
      uVRegion {bIsValid=false, max=<0 0>, min=<0 0>}
     }
     backgroundPressedImage {
      bIsDynamicallyLoaded=false
      drawAs=RoundedBox
      imageType=NoImage
      mirroring=NoMirror
      resourceName=None
      resourceObject=null
      tiling=NoTile
      imageSize=<18 18>
      margin=<ltrb 0 0 0 0>
      outlineSettings {
       bUseBrushTransparency=false
       roundingType=FixedRadius
       width=1
       color {colorUseRule=UseColor_ColorTable, specifiedColor=<rgba 0 0 0 0>}
       cornerRadii=<3 3 3 3>
      }
      tintColor {colorUseRule=UseColor_Specified, specifiedColor=<rgba 0.0048 0.0048 0.0048 1>}
      uVRegion {bIsValid=false, max=<0 0>, min=<0 0>}
     }
     borderBackgroundColor {colorUseRule=UseColor_Specified, specifiedColor=<rgba 1 1 1 1>}
     checkedForeground {colorUseRule=UseColor_Specified, specifiedColor=<rgba 1 1 1 1>}
     checkedHoveredForeground {colorUseRule=UseColor_Specified, specifiedColor=<rgba 1 1 1 1>}
     checkedHoveredImage {
      bIsDynamicallyLoaded=false
      drawAs=RoundedBox
      imageType=FullColor
      mirroring=NoMirror
      resourceName=None
      resourceObject=&checkedsolid
      tiling=NoTile
      imageSize=<128 128>
      margin=<ltrb 0 0 0 0>
      outlineSettings {
       bUseBrushTransparency=true
       roundingType=FixedRadius
       width=1
       color {colorUseRule=UseColor_Specified, specifiedColor=<rgba 1 0.5 0 1>}
       cornerRadii=<4 4 4 4>
      }
      tintColor {colorUseRule=UseColor_Specified, specifiedColor=<rgba 1 0.5 0 1>}
      uVRegion {bIsValid=false, max=<0 0>, min=<0 0>}
     }
     checkedImage {
      bIsDynamicallyLoaded=false
      drawAs=Image
      imageType=FullColor
      mirroring=NoMirror
      resourceName=None
      resourceObject=&checkedsolid
      tiling=NoTile
      imageSize=<24 24>
      margin=<ltrb 0 0 1 1>
      outlineSettings {
       bUseBrushTransparency=true
       roundingType=FixedRadius
       width=1
       color {colorUseRule=UseColor_Specified, specifiedColor=<rgba 0.5 0.25 0 1>}
       cornerRadii=<4 4 4 4>
      }
      tintColor {colorUseRule=UseColor_Specified, specifiedColor=<rgba 0.5 0.25 0 1>}
      uVRegion {bIsValid=false, max=<0 0>, min=<0 0>}
     }
     checkedPressedForeground {colorUseRule=UseColor_Specified, specifiedColor=<rgba 1 1 1 1>}
     checkedPressedImage {
      bIsDynamicallyLoaded=false
      drawAs=RoundedBox
      imageType=FullColor
      mirroring=NoMirror
      resourceName=None
      resourceObject=&checkedsolid
      tiling=NoTile
      imageSize=<128 128>
      margin=<ltrb 0 0 0 0>
      outlineSettings {
       bUseBrushTransparency=false
       roundingType=FixedRadius
       width=1
       color {colorUseRule=UseColor_Specified, specifiedColor=<rgba 0.2 0.1 0 1>}
       cornerRadii=<4 4 4 4>
      }
      tintColor {colorUseRule=UseColor_Specified, specifiedColor=<rgba 0.2 0.1 0 1>}
      uVRegion {bIsValid=false, max=<0 0>, min=<0 0>}
     }
     checkedSlateSound {resourceObject=null}
     foregroundColor {colorUseRule=UseColor_Specified, specifiedColor=<rgba 1 1 1 1>}
     hoveredForeground {colorUseRule=UseColor_Specified, specifiedColor=<rgba 1 1 1 1>}
     hoveredSlateSound {resourceObject=null}
     padding=<ltrb 14 14 14 14>
     pressedForeground {colorUseRule=UseColor_Specified, specifiedColor=<rgba 1 1 1 1>}
     uncheckedHoveredImage {
      bIsDynamicallyLoaded=false
      drawAs=RoundedBox
      imageType=NoImage
      mirroring=NoMirror
      resourceName=None
      resourceObject=null
      tiling=NoTile
      imageSize=<16 16>
      margin=<ltrb 0 0 0 0>
      outlineSettings {
       bUseBrushTransparency=false
       roundingType=FixedRadius
       width=1
       color {colorUseRule=UseColor_Specified, specifiedColor=<rgba 1 0.5 0 1>}
       cornerRadii=<4 4 4 4>
      }
      tintColor {colorUseRule=UseColor_Specified, specifiedColor=<rgba 0.2 0.1 0 0>}
      uVRegion {bIsValid=false, max=<0 0>, min=<0 0>}
     }
     uncheckedImage {
      bIsDynamicallyLoaded=false
      drawAs=RoundedBox
      imageType=NoImage
      mirroring=NoMirror
      resourceName=None
      resourceObject=null
      tiling=NoTile
      imageSize=<16 16>
      margin=<ltrb 0 0 0 0>
      outlineSettings {
       bUseBrushTransparency=false
       roundingType=FixedRadius
       width=1
       color {colorUseRule=UseColor_Specified, specifiedColor=<rgba 0.5 0.25 0 1>}
       cornerRadii=<4 4 4 4>
      }
      tintColor {colorUseRule=UseColor_Specified, specifiedColor=<rgba 0.2 0.1 0 0>}
      uVRegion {bIsValid=false, max=<0 0>, min=<0 0>}
     }
     uncheckedPressedImage {
      bIsDynamicallyLoaded=false
      drawAs=RoundedBox
      imageType=NoImage
      mirroring=NoMirror
      resourceName=None
      resourceObject=null
      tiling=NoTile
      imageSize=<16 16>
      margin=<ltrb 0 0 0 0>
      outlineSettings {
       bUseBrushTransparency=false
       roundingType=FixedRadius
       width=1
       color {colorUseRule=UseColor_Specified, specifiedColor=<rgba 0.2 0.1 0 1>}
       cornerRadii=<4 4 4 4>
      }
      tintColor {colorUseRule=UseColor_Specified, specifiedColor=<rgba 0.2 0.1 0 0>}
      uVRegion {bIsValid=false, max=<0 0>, min=<0 0>}
     }
     uncheckedSlateSound {resourceObject=null}
     undeterminedForeground {colorUseRule=UseColor_Specified, specifiedColor=<rgba 1 1 1 1>}
     undeterminedHoveredImage {
      bIsDynamicallyLoaded=false
      drawAs=Image
      imageType=Vector
      mirroring=NoMirror
      resourceName=../../../Engine/Content/Slate/Starship/CoreWidgets/CheckBox/indeterminate.svg
      resourceObject=null
      tiling=NoTile
      imageSize=<16 16>
      margin=<ltrb 0 0 0 0>
      outlineSettings {
       bUseBrushTransparency=false
       roundingType=HalfHeightRadius
       width=0
       color {colorUseRule=UseColor_Specified, specifiedColor=<rgba 0 0 0 0>}
       cornerRadii=<0 0 0 0>
      }
      tintColor {colorUseRule=UseColor_Specified, specifiedColor=<rgba 0.0044 0.2384 1 1>}
      uVRegion {bIsValid=false, max=<0 0>, min=<0 0>}
     }
     undeterminedImage {
      bIsDynamicallyLoaded=false
      drawAs=Image
      imageType=Vector
      mirroring=NoMirror
      resourceName=../../../Engine/Content/Slate/Starship/CoreWidgets/CheckBox/indeterminate.svg
      resourceObject=null
      tiling=NoTile
      imageSize=<16 16>
      margin=<ltrb 0 0 0 0>
      outlineSettings {
       bUseBrushTransparency=false
       roundingType=HalfHeightRadius
       width=0
       color {colorUseRule=UseColor_Specified, specifiedColor=<rgba 0 0 0 0>}
       cornerRadii=<0 0 0 0>
      }
      tintColor {colorUseRule=UseColor_Specified, specifiedColor=<rgba 0 0.162 0.7454 1>}
      uVRegion {bIsValid=false, max=<0 0>, min=<0 0>}
     }
     undeterminedPressedImage {
      bIsDynamicallyLoaded=false
      drawAs=Image
      imageType=Vector
      mirroring=NoMirror
      resourceName=../../../Engine/Content/Slate/Starship/CoreWidgets/CheckBox/indeterminate.svg
      resourceObject=null
      tiling=NoTile
      imageSize=<16 16>
      margin=<ltrb 0 0 0 0>
      outlineSettings {
       bUseBrushTransparency=false
       roundingType=HalfHeightRadius
       width=0
       color {colorUseRule=UseColor_Specified, specifiedColor=<rgba 0 0 0 0>}
       cornerRadii=<0 0 0 0>
      }
      tintColor {colorUseRule=UseColor_Specified, specifiedColor=<rgba 0 0.0802 0.3515 1>}
      uVRegion {bIsValid=false, max=<0 0>, min=<0 0>}
     }
    }
   }
   slot {class=HorizontalBoxSlot, properties {padding=<ltrb 4 4 4 4>, size {sizeRule=Fill, value=1}}}
  }
  HorizontalBox_3 {
   widgetClassPath=&HorizontalBox
   parent=WrapBox_0
   widget=HorizontalBox_3
   slot {class=WrapBoxSlot, properties {bFillEmptySpace=true, fillSpanWhenLessThan=256, horizontalAlignment=HAlign_Right}}
  }
  btnAdvancedSearch {
   widgetClassPath=&EditorUtilityButton
   parent=HorizontalBox_3
   widget=btnAdvancedSearch
   properties {widgetStyle {
     clickedSlateSound {resourceObject=null}
     disabled {
      bIsDynamicallyLoaded=false
      drawAs=RoundedBox
      imageType=NoImage
      mirroring=NoMirror
      resourceName=None
      resourceObject=null
      tiling=NoTile
      imageSize=<0 0>
      margin=<ltrb 0 0 0 0>
      outlineSettings {
       bUseBrushTransparency=false
       roundingType=FixedRadius
       width=1
       color {colorUseRule=UseColor_ColorTable, specifiedColor=<rgba 0 0 0 0>}
       cornerRadii=<4 4 4 4>
      }
      tintColor {colorUseRule=UseColor_Specified, specifiedColor=<rgba 0.0395 0.0395 0.0395 1>}
      uVRegion {bIsValid=false, max=<0 0>, min=<0 0>}
     }
     disabledForeground {colorUseRule=UseColor_ColorTable, specifiedColor=<rgba 0 0 0 0>}
     hovered {
      bIsDynamicallyLoaded=false
      drawAs=RoundedBox
      imageType=NoImage
      mirroring=NoMirror
      resourceName=None
      resourceObject=null
      tiling=NoTile
      imageSize=<0 0>
      margin=<ltrb 0 0 0 0>
      outlineSettings {
       bUseBrushTransparency=false
       roundingType=FixedRadius
       width=1
       color {colorUseRule=UseColor_Specified, specifiedColor=<rgba 0.527 0.527 0.527 0>}
       cornerRadii=<4 4 4 4>
      }
      tintColor {colorUseRule=UseColor_Specified, specifiedColor=<rgba 0.0048 0.0048 0.0048 1>}
      uVRegion {bIsValid=false, max=<0 0>, min=<0 0>}
     }
     hoveredForeground {colorUseRule=UseColor_ColorTable, specifiedColor=<rgba 0 0 0 0>}
     hoveredSlateSound {resourceObject=null}
     normal {
      bIsDynamicallyLoaded=false
      drawAs=RoundedBox
      imageType=FullColor
      mirroring=NoMirror
      resourceName=None
      resourceObject=null
      tiling=NoTile
      imageSize=<24 24>
      margin=<ltrb 0 0 0 0>
      outlineSettings {
       bUseBrushTransparency=false
       roundingType=FixedRadius
       width=1
       color {colorUseRule=UseColor_ColorTable, specifiedColor=<rgba 0 0 0 0>}
       cornerRadii=<4 4 4 4>
      }
      tintColor {colorUseRule=UseColor_Specified, specifiedColor=<rgba 0.0048 0.0048 0.0048 1>}
      uVRegion {bIsValid=false, max=<0 0>, min=<0 0>}
     }
     normalForeground {colorUseRule=UseColor_ColorTable, specifiedColor=<rgba 0 0 0 0>}
     normalPadding=<ltrb 0 0 0 0>
     pressed {
      bIsDynamicallyLoaded=false
      drawAs=RoundedBox
      imageType=NoImage
      mirroring=NoMirror
      resourceName=None
      resourceObject=null
      tiling=NoTile
      imageSize=<32 32>
      margin=<ltrb 0 0 0 0>
      outlineSettings {
       bUseBrushTransparency=false
       roundingType=FixedRadius
       width=1
       color {colorUseRule=UseColor_ColorTable, specifiedColor=<rgba 0 0 0 0>}
       cornerRadii=<4 4 4 4>
      }
      tintColor {colorUseRule=UseColor_Specified, specifiedColor=<rgba 0.0284 0.0284 0.0284 1>}
      uVRegion {bIsValid=false, max=<0 0>, min=<0 0>}
     }
     pressedForeground {colorUseRule=UseColor_ColorTable, specifiedColor=<rgba 0 0 0 0>}
     pressedPadding=<ltrb 0 0 0 0>
     pressedSlateSound {resourceObject=null}
    }}
   slot {class=HorizontalBoxSlot, properties {horizontalAlignment=HAlign_Left, verticalAlignment=VAlign_Center}}
  }
  HorizontalBox_4 {
   widgetClassPath=&HorizontalBox
   parent=btnAdvancedSearch
   widget=HorizontalBox_4
   slot {class=ButtonSlot, properties {horizontalAlignment=HAlign_Fill, verticalAlignment=VAlign_Fill, padding=<ltrb 4 4 4 4>}}
  }
  Image {
   widgetClassPath=&Image
   parent=HorizontalBox_4
   widget=Image
   properties {brush {
     bIsDynamicallyLoaded=false
     drawAs=Image
     imageType=FullColor
     mirroring=NoMirror
     resourceName=None
     resourceObject=/AnimationLibrary/resources/icons/magnifyingglasssolid
     tiling=NoTile
     imageSize=<20 20>
     margin=<ltrb 0 0 0 0>
     outlineSettings {
      bUseBrushTransparency=false
      roundingType=HalfHeightRadius
      width=0
      color {colorUseRule=UseColor_Specified, specifiedColor=<rgba 0 0 0 0>}
      cornerRadii=<0 0 0 0>
     }
     tintColor {colorUseRule=UseColor_Specified, specifiedColor=<rgba 0.0742 0.0742 0.0742 1>}
     uVRegion {bIsValid=false, max=<0 0>, min=<0 0>}
    }}
   slot {class=HorizontalBoxSlot, properties {horizontalAlignment=HAlign_Center, verticalAlignment=VAlign_Center}}
  }
  Image_59 {
   widgetClassPath=&Image
   parent=HorizontalBox_4
   widget=Image_59
   properties {brush {
     bIsDynamicallyLoaded=false
     drawAs=Image
     imageType=FullColor
     mirroring=NoMirror
     resourceName=None
     resourceObject=/AnimationLibrary/resources/icons/chevrondownsolid
     tiling=NoTile
     imageSize=<16 16>
     margin=<ltrb 0 0 0 0>
     outlineSettings {
      bUseBrushTransparency=false
      roundingType=HalfHeightRadius
      width=0
      color {colorUseRule=UseColor_Specified, specifiedColor=<rgba 0 0 0 0>}
      cornerRadii=<0 0 0 0>
     }
     tintColor {colorUseRule=UseColor_Specified, specifiedColor=<rgba 0.0742 0.0742 0.0742 1>}
     uVRegion {bIsValid=false, max=<0 0>, min=<0 0>}
    }}
   slot {class=HorizontalBoxSlot, properties {horizontalAlignment=HAlign_Center, verticalAlignment=VAlign_Center, padding=<ltrb 1 2 2 2>}}
  }
  SizeBox_145 {
   widgetClassPath=&SizeBox
   parent=HorizontalBox_3
   widget=SizeBox_145
   properties {
    bOverride_MaxDesiredHeight=true
    bOverride_MaxDesiredWidth=true
    bOverride_MinDesiredHeight=true
    bOverride_MinDesiredWidth=true
    maxDesiredHeight=28
    maxDesiredWidth=400
    minDesiredHeight=28
    minDesiredWidth=60
   }
   slot {
    class=HorizontalBoxSlot
    properties {horizontalAlignment=HAlign_Center, verticalAlignment=VAlign_Center, size {sizeRule=Automatic, value=0.5}}
   }
  }
  SearchBox {
   widgetClassPath=&EditorUtilityEditableTextBox
   bIsVariable=true
   parent=SizeBox_145
   widget=SearchBox
   properties {
    hintText=Search by name...
    minimumDesiredWidth=252
    toolTipText=Needs at least 3 characters
    widgetStyle {
     backgroundColor {colorUseRule=UseColor_Specified, specifiedColor=<rgba 1 1 1 1>}
     backgroundImageFocused {
      bIsDynamicallyLoaded=false
      drawAs=RoundedBox
      imageType=NoImage
      mirroring=NoMirror
      resourceName=None
      resourceObject=null
      tiling=NoTile
      imageSize=<0 0>
      margin=<ltrb 0 0 0 0>
      outlineSettings {
       bUseBrushTransparency=false
       roundingType=FixedRadius
       width=1
       color {colorUseRule=UseColor_ColorTable, specifiedColor=<rgba 0 0 0 0>}
       cornerRadii=<4 4 4 4>
      }
      tintColor {colorUseRule=UseColor_Specified, specifiedColor=<rgba 0.0048 0.0048 0.0048 1>}
      uVRegion {bIsValid=false, max=<0 0>, min=<0 0>}
     }
     backgroundImageHovered {
      bIsDynamicallyLoaded=false
      drawAs=RoundedBox
      imageType=NoImage
      mirroring=NoMirror
      resourceName=None
      resourceObject=null
      tiling=NoTile
      imageSize=<0 0>
      margin=<ltrb 0 0 0 0>
      outlineSettings {
       bUseBrushTransparency=false
       roundingType=FixedRadius
       width=1
       color {colorUseRule=UseColor_ColorTable, specifiedColor=<rgba 0 0 0 0>}
       cornerRadii=<4 4 4 4>
      }
      tintColor {colorUseRule=UseColor_Specified, specifiedColor=<rgba 0.0048 0.0048 0.0048 1>}
      uVRegion {bIsValid=false, max=<0 0>, min=<0 0>}
     }
     backgroundImageNormal {
      bIsDynamicallyLoaded=false
      drawAs=RoundedBox
      imageType=NoImage
      mirroring=NoMirror
      resourceName=None
      resourceObject=null
      tiling=NoTile
      imageSize=<0 0>
      margin=<ltrb 0 0 0 0>
      outlineSettings {
       bUseBrushTransparency=false
       roundingType=FixedRadius
       width=1
       color {colorUseRule=UseColor_ColorTable, specifiedColor=<rgba 0 0 0 0>}
       cornerRadii=<4 4 4 4>
      }
      tintColor {colorUseRule=UseColor_Specified, specifiedColor=<rgba 0.0048 0.0048 0.0048 1>}
      uVRegion {bIsValid=false, max=<0 0>, min=<0 0>}
     }
     backgroundImageReadOnly {
      bIsDynamicallyLoaded=false
      drawAs=RoundedBox
      imageType=NoImage
      mirroring=NoMirror
      resourceName=None
      resourceObject=null
      tiling=NoTile
      imageSize=<0 0>
      margin=<ltrb 0 0 0 0>
      outlineSettings {
       bUseBrushTransparency=false
       roundingType=FixedRadius
       width=1
       color {colorUseRule=UseColor_ColorTable, specifiedColor=<rgba 0 0 0 0>}
       cornerRadii=<4 4 4 4>
      }
      tintColor {colorUseRule=UseColor_Specified, specifiedColor=<rgba 0.0284 0.0284 0.0284 1>}
      uVRegion {bIsValid=false, max=<0 0>, min=<0 0>}
     }
     focusedForegroundColor {colorUseRule=UseColor_Specified, specifiedColor=<rgba 1 1 1 1>}
     foregroundColor {colorUseRule=UseColor_Specified, specifiedColor=<rgba 0.5271 0.5271 0.5271 1>}
     hScrollBarPadding=<ltrb 0 0 0 0>
     padding=<ltrb 4 4 4 4>
     readOnlyForegroundColor {colorUseRule=UseColor_Foreground, specifiedColor=<rgba 1 0 1 1>}
     scrollBarStyle {
      thickness=8
      draggedThumbImage {
       bIsDynamicallyLoaded=false
       drawAs=RoundedBox
       imageType=NoImage
       mirroring=NoMirror
       resourceName=None
       resourceObject=null
       tiling=NoTile
       imageSize=<0 0>
       margin=<ltrb 0 0 0 0>
       outlineSettings {
        bUseBrushTransparency=false
        roundingType=FixedRadius
        width=0
        color {colorUseRule=UseColor_Specified, specifiedColor=<rgba 0 0 0 0>}
        cornerRadii=<4 4 4 4>
       }
       tintColor {colorUseRule=UseColor_Specified, specifiedColor=<rgba 0.2159 0.2159 0.2159 1>}
       uVRegion {bIsValid=false, max=<0 0>, min=<0 0>}
      }
      horizontalBackgroundImage {
       bIsDynamicallyLoaded=false
       drawAs=NoDrawType
       imageType=NoImage
       mirroring=NoMirror
       resourceName=None
       resourceObject=null
       tiling=NoTile
       imageSize=<0 0>
       margin=<ltrb 0 0 0 0>
       outlineSettings {
        bUseBrushTransparency=false
        roundingType=HalfHeightRadius
        width=0
        color {colorUseRule=UseColor_Specified, specifiedColor=<rgba 0 0 0 0>}
        cornerRadii=<0 0 0 0>
       }
       tintColor {colorUseRule=UseColor_Specified, specifiedColor=<rgba 1 1 1 1>}
       uVRegion {bIsValid=false, max=<0 0>, min=<0 0>}
      }
      horizontalBottomSlotImage {
       bIsDynamicallyLoaded=false
       drawAs=NoDrawType
       imageType=NoImage
       mirroring=NoMirror
       resourceName=None
       resourceObject=null
       tiling=NoTile
       imageSize=<0 0>
       margin=<ltrb 0 0 0 0>
       outlineSettings {
        bUseBrushTransparency=false
        roundingType=HalfHeightRadius
        width=0
        color {colorUseRule=UseColor_Specified, specifiedColor=<rgba 0 0 0 0>}
        cornerRadii=<0 0 0 0>
       }
       tintColor {colorUseRule=UseColor_Specified, specifiedColor=<rgba 1 1 1 1>}
       uVRegion {bIsValid=false, max=<0 0>, min=<0 0>}
      }
      horizontalTopSlotImage {
       bIsDynamicallyLoaded=false
       drawAs=NoDrawType
       imageType=NoImage
       mirroring=NoMirror
       resourceName=None
       resourceObject=null
       tiling=NoTile
       imageSize=<0 0>
       margin=<ltrb 0 0 0 0>
       outlineSettings {
        bUseBrushTransparency=false
        roundingType=HalfHeightRadius
        width=0
        color {colorUseRule=UseColor_Specified, specifiedColor=<rgba 0 0 0 0>}
        cornerRadii=<0 0 0 0>
       }
       tintColor {colorUseRule=UseColor_Specified, specifiedColor=<rgba 1 1 1 1>}
       uVRegion {bIsValid=false, max=<0 0>, min=<0 0>}
      }
      hoveredThumbImage {
       bIsDynamicallyLoaded=false
       drawAs=RoundedBox
       imageType=NoImage
       mirroring=NoMirror
       resourceName=None
       resourceObject=null
       tiling=NoTile
       imageSize=<0 0>
       margin=<ltrb 0 0 0 0>
       outlineSettings {
        bUseBrushTransparency=false
        roundingType=FixedRadius
        width=0
        color {colorUseRule=UseColor_Specified, specifiedColor=<rgba 0 0 0 0>}
        cornerRadii=<4 4 4 4>
       }
       tintColor {colorUseRule=UseColor_Specified, specifiedColor=<rgba 0.2159 0.2159 0.2159 1>}
       uVRegion {bIsValid=false, max=<0 0>, min=<0 0>}
      }
      normalThumbImage {
       bIsDynamicallyLoaded=false
       drawAs=RoundedBox
       imageType=NoImage
       mirroring=NoMirror
       resourceName=None
       resourceObject=null
       tiling=NoTile
       imageSize=<0 0>
       margin=<ltrb 0 0 0 0>
       outlineSettings {
        bUseBrushTransparency=false
        roundingType=FixedRadius
        width=0
        color {colorUseRule=UseColor_Specified, specifiedColor=<rgba 0 0 0 0>}
        cornerRadii=<4 4 4 4>
       }
       tintColor {colorUseRule=UseColor_Specified, specifiedColor=<rgba 0.0953 0.0953 0.0953 1>}
       uVRegion {bIsValid=false, max=<0 0>, min=<0 0>}
      }
      verticalBackgroundImage {
       bIsDynamicallyLoaded=false
       drawAs=NoDrawType
       imageType=NoImage
       mirroring=NoMirror
       resourceName=None
       resourceObject=null
       tiling=NoTile
       imageSize=<0 0>
       margin=<ltrb 0 0 0 0>
       outlineSettings {
        bUseBrushTransparency=false
        roundingType=HalfHeightRadius
        width=0
        color {colorUseRule=UseColor_Specified, specifiedColor=<rgba 0 0 0 0>}
        cornerRadii=<0 0 0 0>
       }
       tintColor {colorUseRule=UseColor_Specified, specifiedColor=<rgba 1 1 1 1>}
       uVRegion {bIsValid=false, max=<0 0>, min=<0 0>}
      }
      verticalBottomSlotImage {
       bIsDynamicallyLoaded=false
       drawAs=NoDrawType
       imageType=NoImage
       mirroring=NoMirror
       resourceName=None
       resourceObject=null
       tiling=NoTile
       imageSize=<0 0>
       margin=<ltrb 0 0 0 0>
       outlineSettings {
        bUseBrushTransparency=false
        roundingType=HalfHeightRadius
        width=0
        color {colorUseRule=UseColor_Specified, specifiedColor=<rgba 0 0 0 0>}
        cornerRadii=<0 0 0 0>
       }
       tintColor {colorUseRule=UseColor_Specified, specifiedColor=<rgba 1 1 1 1>}
       uVRegion {bIsValid=false, max=<0 0>, min=<0 0>}
      }
      verticalTopSlotImage {
       bIsDynamicallyLoaded=false
       drawAs=NoDrawType
       imageType=NoImage
       mirroring=NoMirror
       resourceName=None
       resourceObject=null
       tiling=NoTile
       imageSize=<0 0>
       margin=<ltrb 0 0 0 0>
       outlineSettings {
        bUseBrushTransparency=false
        roundingType=HalfHeightRadius
        width=0
        color {colorUseRule=UseColor_Specified, specifiedColor=<rgba 0 0 0 0>}
        cornerRadii=<0 0 0 0>
       }
       tintColor {colorUseRule=UseColor_Specified, specifiedColor=<rgba 1 1 1 1>}
       uVRegion {bIsValid=false, max=<0 0>, min=<0 0>}
      }
     }
     textStyle {
      overflowPolicy=Clip
      transformPolicy=None
      colorAndOpacity {colorUseRule=UseColor_Foreground, specifiedColor=<rgba 1 0 1 1>}
      font {
       bForceMonospaced=false
       bMaterialIsStencil=false
       fontMaterial=null
       fontObject=&Roboto
       letterSpacing=0
       monospacedWidth=1
       size=12.75
       skewAmount=0
       typefaceFontName=Regular
       outlineSettings {
        bApplyOutlineToDropShadows=false
        bMiteredCorners=false
        bSeparateFillAlpha=false
        outlineMaterial=null
        outlineSize=0
        outlineColor=<rgba 0 0 0 1>
       }
      }
      highlightColor {colorUseRule=UseColor_ColorTable, specifiedColor=<rgba 0 0 0 0>}
      highlightShape {
       bIsDynamicallyLoaded=false
       drawAs=Image
       imageType=NoImage
       mirroring=NoMirror
       resourceName=None
       resourceObject=null
       tiling=NoTile
       imageSize=<0 0>
       margin=<ltrb 0 0 0 0>
       outlineSettings {
        bUseBrushTransparency=false
        roundingType=HalfHeightRadius
        width=0
        color {colorUseRule=UseColor_Specified, specifiedColor=<rgba 0 0 0 0>}
        cornerRadii=<0 0 0 0>
       }
       tintColor {colorUseRule=UseColor_ColorTable, specifiedColor=<rgba 0 0 0 0>}
       uVRegion {bIsValid=false, max=<0 0>, min=<0 0>}
      }
      selectedBackgroundColor {colorUseRule=UseColor_ColorTable, specifiedColor=<rgba 0 0 0 0>}
      shadowColorAndOpacity=<rgba 0 0 0 1>
      shadowOffset=<0 0>
      strikeBrush {
       bIsDynamicallyLoaded=false
       drawAs=Image
       imageType=NoImage
       mirroring=NoMirror
       resourceName=None
       resourceObject=null
       tiling=NoTile
       imageSize=<32 32>
       margin=<ltrb 0 0 0 0>
       outlineSettings {
        bUseBrushTransparency=false
        roundingType=HalfHeightRadius
        width=0
        color {colorUseRule=UseColor_Specified, specifiedColor=<rgba 0 0 0 0>}
        cornerRadii=<0 0 0 0>
       }
       tintColor {colorUseRule=UseColor_Specified, specifiedColor=<rgba 1 1 1 1>}
       uVRegion {bIsValid=false, max=<0 0>, min=<0 0>}
      }
      underlineBrush {
       bIsDynamicallyLoaded=false
       drawAs=Image
       imageType=NoImage
       mirroring=NoMirror
       resourceName=None
       resourceObject=null
       tiling=NoTile
       imageSize=<32 32>
       margin=<ltrb 0 0 0 0>
       outlineSettings {
        bUseBrushTransparency=false
        roundingType=HalfHeightRadius
        width=0
        color {colorUseRule=UseColor_Specified, specifiedColor=<rgba 0 0 0 0>}
        cornerRadii=<0 0 0 0>
       }
       tintColor {colorUseRule=UseColor_Specified, specifiedColor=<rgba 1 1 1 1>}
       uVRegion {bIsValid=false, max=<0 0>, min=<0 0>}
      }
     }
     vScrollBarPadding=<ltrb 0 0 0 0>
    }
   }
   slot {class=SizeBoxSlot}
  }
  VB_Advanced {
   widgetClassPath=&VerticalBox
   bIsVariable=true
   parent=Main
   widget=VB_Advanced
   properties {visibility=Collapsed}
   slot {class=VerticalBoxSlot, properties {verticalAlignment=VAlign_Top, size {sizeRule=Automatic, value=0.5}}}
  }
  SearchByTags {
   widgetClassPath=&EditorUtilityEditableTextBox
   bIsVariable=true
   parent=VB_Advanced
   widget=SearchByTags
   properties {
    bOverride_Cursor=true
    clipping=ClipToBounds
    hintText=Search by tags...
    minimumDesiredWidth=300
    toolTipText="Needs at least 3 characters, tags needs to be separated by a comma"
    widgetStyle {
     backgroundColor {colorUseRule=UseColor_Specified, specifiedColor=<rgba 1 1 1 1>}
     backgroundImageFocused {
      bIsDynamicallyLoaded=false
      drawAs=RoundedBox
      imageType=NoImage
      mirroring=NoMirror
      resourceName=None
      resourceObject=null
      tiling=NoTile
      imageSize=<0 0>
      margin=<ltrb 0 0 0 0>
      outlineSettings {
       bUseBrushTransparency=false
       roundingType=FixedRadius
       width=1
       color {colorUseRule=UseColor_ColorTable, specifiedColor=<rgba 0 0 0 0>}
       cornerRadii=<4 4 4 4>
      }
      tintColor {colorUseRule=UseColor_Specified, specifiedColor=<rgba 0.0048 0.0048 0.0048 1>}
      uVRegion {bIsValid=false, max=<0 0>, min=<0 0>}
     }
     backgroundImageHovered {
      bIsDynamicallyLoaded=false
      drawAs=RoundedBox
      imageType=NoImage
      mirroring=NoMirror
      resourceName=None
      resourceObject=null
      tiling=NoTile
      imageSize=<0 0>
      margin=<ltrb 0 0 0 0>
      outlineSettings {
       bUseBrushTransparency=false
       roundingType=FixedRadius
       width=1
       color {colorUseRule=UseColor_ColorTable, specifiedColor=<rgba 0 0 0 0>}
       cornerRadii=<4 4 4 4>
      }
      tintColor {colorUseRule=UseColor_Specified, specifiedColor=<rgba 0.0048 0.0048 0.0048 1>}
      uVRegion {bIsValid=false, max=<0 0>, min=<0 0>}
     }
     backgroundImageNormal {
      bIsDynamicallyLoaded=false
      drawAs=RoundedBox
      imageType=NoImage
      mirroring=NoMirror
      resourceName=None
      resourceObject=null
      tiling=NoTile
      imageSize=<0 0>
      margin=<ltrb 0 0 0 0>
      outlineSettings {
       bUseBrushTransparency=false
       roundingType=FixedRadius
       width=1
       color {colorUseRule=UseColor_ColorTable, specifiedColor=<rgba 0 0 0 0>}
       cornerRadii=<4 4 4 4>
      }
      tintColor {colorUseRule=UseColor_Specified, specifiedColor=<rgba 0.0048 0.0048 0.0048 1>}
      uVRegion {bIsValid=false, max=<0 0>, min=<0 0>}
     }
     backgroundImageReadOnly {
      bIsDynamicallyLoaded=false
      drawAs=RoundedBox
      imageType=NoImage
      mirroring=NoMirror
      resourceName=None
      resourceObject=null
      tiling=NoTile
      imageSize=<0 0>
      margin=<ltrb 0 0 0 0>
      outlineSettings {
       bUseBrushTransparency=false
       roundingType=FixedRadius
       width=1
       color {colorUseRule=UseColor_ColorTable, specifiedColor=<rgba 0 0 0 0>}
       cornerRadii=<4 4 4 4>
      }
      tintColor {colorUseRule=UseColor_Specified, specifiedColor=<rgba 0.0284 0.0284 0.0284 1>}
      uVRegion {bIsValid=false, max=<0 0>, min=<0 0>}
     }
     focusedForegroundColor {colorUseRule=UseColor_Specified, specifiedColor=<rgba 1 1 1 1>}
     foregroundColor {colorUseRule=UseColor_Specified, specifiedColor=<rgba 0.5271 0.5271 0.5271 1>}
     hScrollBarPadding=<ltrb 0 0 0 0>
     padding=<ltrb 4 4 4 4>
     readOnlyForegroundColor {colorUseRule=UseColor_Foreground, specifiedColor=<rgba 1 0 1 1>}
     scrollBarStyle {
      thickness=8
      draggedThumbImage {
       bIsDynamicallyLoaded=false
       drawAs=RoundedBox
       imageType=NoImage
       mirroring=NoMirror
       resourceName=None
       resourceObject=null
       tiling=NoTile
       imageSize=<0 0>
       margin=<ltrb 0 0 0 0>
       outlineSettings {
        bUseBrushTransparency=false
        roundingType=FixedRadius
        width=0
        color {colorUseRule=UseColor_Specified, specifiedColor=<rgba 0 0 0 0>}
        cornerRadii=<4 4 4 4>
       }
       tintColor {colorUseRule=UseColor_Specified, specifiedColor=<rgba 0.2159 0.2159 0.2159 1>}
       uVRegion {bIsValid=false, max=<0 0>, min=<0 0>}
      }
      horizontalBackgroundImage {
       bIsDynamicallyLoaded=false
       drawAs=NoDrawType
       imageType=NoImage
       mirroring=NoMirror
       resourceName=None
       resourceObject=null
       tiling=NoTile
       imageSize=<0 0>
       margin=<ltrb 0 0 0 0>
       outlineSettings {
        bUseBrushTransparency=false
        roundingType=HalfHeightRadius
        width=0
        color {colorUseRule=UseColor_Specified, specifiedColor=<rgba 0 0 0 0>}
        cornerRadii=<0 0 0 0>
       }
       tintColor {colorUseRule=UseColor_Specified, specifiedColor=<rgba 1 1 1 1>}
       uVRegion {bIsValid=false, max=<0 0>, min=<0 0>}
      }
      horizontalBottomSlotImage {
       bIsDynamicallyLoaded=false
       drawAs=NoDrawType
       imageType=NoImage
       mirroring=NoMirror
       resourceName=None
       resourceObject=null
       tiling=NoTile
       imageSize=<0 0>
       margin=<ltrb 0 0 0 0>
       outlineSettings {
        bUseBrushTransparency=false
        roundingType=HalfHeightRadius
        width=0
        color {colorUseRule=UseColor_Specified, specifiedColor=<rgba 0 0 0 0>}
        cornerRadii=<0 0 0 0>
       }
       tintColor {colorUseRule=UseColor_Specified, specifiedColor=<rgba 1 1 1 1>}
       uVRegion {bIsValid=false, max=<0 0>, min=<0 0>}
      }
      horizontalTopSlotImage {
       bIsDynamicallyLoaded=false
       drawAs=NoDrawType
       imageType=NoImage
       mirroring=NoMirror
       resourceName=None
       resourceObject=null
       tiling=NoTile
       imageSize=<0 0>
       margin=<ltrb 0 0 0 0>
       outlineSettings {
        bUseBrushTransparency=false
        roundingType=HalfHeightRadius
        width=0
        color {colorUseRule=UseColor_Specified, specifiedColor=<rgba 0 0 0 0>}
        cornerRadii=<0 0 0 0>
       }
       tintColor {colorUseRule=UseColor_Specified, specifiedColor=<rgba 1 1 1 1>}
       uVRegion {bIsValid=false, max=<0 0>, min=<0 0>}
      }
      hoveredThumbImage {
       bIsDynamicallyLoaded=false
       drawAs=RoundedBox
       imageType=NoImage
       mirroring=NoMirror
       resourceName=None
       resourceObject=null
       tiling=NoTile
       imageSize=<0 0>
       margin=<ltrb 0 0 0 0>
       outlineSettings {
        bUseBrushTransparency=false
        roundingType=FixedRadius
        width=0
        color {colorUseRule=UseColor_Specified, specifiedColor=<rgba 0 0 0 0>}
        cornerRadii=<4 4 4 4>
       }
       tintColor {colorUseRule=UseColor_Specified, specifiedColor=<rgba 0.2159 0.2159 0.2159 1>}
       uVRegion {bIsValid=false, max=<0 0>, min=<0 0>}
      }
      normalThumbImage {
       bIsDynamicallyLoaded=false
       drawAs=RoundedBox
       imageType=NoImage
       mirroring=NoMirror
       resourceName=None
       resourceObject=null
       tiling=NoTile
       imageSize=<0 0>
       margin=<ltrb 0 0 0 0>
       outlineSettings {
        bUseBrushTransparency=false
        roundingType=FixedRadius
        width=0
        color {colorUseRule=UseColor_Specified, specifiedColor=<rgba 0 0 0 0>}
        cornerRadii=<4 4 4 4>
       }
       tintColor {colorUseRule=UseColor_Specified, specifiedColor=<rgba 0.0953 0.0953 0.0953 1>}
       uVRegion {bIsValid=false, max=<0 0>, min=<0 0>}
      }
      verticalBackgroundImage {
       bIsDynamicallyLoaded=false
       drawAs=NoDrawType
       imageType=NoImage
       mirroring=NoMirror
       resourceName=None
       resourceObject=null
       tiling=NoTile
       imageSize=<0 0>
       margin=<ltrb 0 0 0 0>
       outlineSettings {
        bUseBrushTransparency=false
        roundingType=HalfHeightRadius
        width=0
        color {colorUseRule=UseColor_Specified, specifiedColor=<rgba 0 0 0 0>}
        cornerRadii=<0 0 0 0>
       }
       tintColor {colorUseRule=UseColor_Specified, specifiedColor=<rgba 1 1 1 1>}
       uVRegion {bIsValid=false, max=<0 0>, min=<0 0>}
      }
      verticalBottomSlotImage {
       bIsDynamicallyLoaded=false
       drawAs=NoDrawType
       imageType=NoImage
       mirroring=NoMirror
       resourceName=None
       resourceObject=null
       tiling=NoTile
       imageSize=<0 0>
       margin=<ltrb 0 0 0 0>
       outlineSettings {
        bUseBrushTransparency=false
        roundingType=HalfHeightRadius
        width=0
        color {colorUseRule=UseColor_Specified, specifiedColor=<rgba 0 0 0 0>}
        cornerRadii=<0 0 0 0>
       }
       tintColor {colorUseRule=UseColor_Specified, specifiedColor=<rgba 1 1 1 1>}
       uVRegion {bIsValid=false, max=<0 0>, min=<0 0>}
      }
      verticalTopSlotImage {
       bIsDynamicallyLoaded=false
       drawAs=NoDrawType
       imageType=NoImage
       mirroring=NoMirror
       resourceName=None
       resourceObject=null
       tiling=NoTile
       imageSize=<0 0>
       margin=<ltrb 0 0 0 0>
       outlineSettings {
        bUseBrushTransparency=false
        roundingType=HalfHeightRadius
        width=0
        color {colorUseRule=UseColor_Specified, specifiedColor=<rgba 0 0 0 0>}
        cornerRadii=<0 0 0 0>
       }
       tintColor {colorUseRule=UseColor_Specified, specifiedColor=<rgba 1 1 1 1>}
       uVRegion {bIsValid=false, max=<0 0>, min=<0 0>}
      }
     }
     textStyle {
      overflowPolicy=Clip
      transformPolicy=None
      colorAndOpacity {colorUseRule=UseColor_Foreground, specifiedColor=<rgba 1 0 1 1>}
      font {
       bForceMonospaced=false
       bMaterialIsStencil=false
       fontMaterial=null
       fontObject=&Roboto
       letterSpacing=0
       monospacedWidth=1
       size=12
       skewAmount=0
       typefaceFontName=Regular
       outlineSettings {
        bApplyOutlineToDropShadows=false
        bMiteredCorners=false
        bSeparateFillAlpha=false
        outlineMaterial=null
        outlineSize=0
        outlineColor=<rgba 0 0 0 1>
       }
      }
      highlightColor {colorUseRule=UseColor_ColorTable, specifiedColor=<rgba 0 0 0 0>}
      highlightShape {
       bIsDynamicallyLoaded=false
       drawAs=Image
       imageType=NoImage
       mirroring=NoMirror
       resourceName=None
       resourceObject=null
       tiling=NoTile
       imageSize=<0 0>
       margin=<ltrb 0 0 0 0>
       outlineSettings {
        bUseBrushTransparency=false
        roundingType=HalfHeightRadius
        width=0
        color {colorUseRule=UseColor_Specified, specifiedColor=<rgba 0 0 0 0>}
        cornerRadii=<0 0 0 0>
       }
       tintColor {colorUseRule=UseColor_ColorTable, specifiedColor=<rgba 0 0 0 0>}
       uVRegion {bIsValid=false, max=<0 0>, min=<0 0>}
      }
      selectedBackgroundColor {colorUseRule=UseColor_ColorTable, specifiedColor=<rgba 0 0 0 0>}
      shadowColorAndOpacity=<rgba 0 0 0 1>
      shadowOffset=<0 0>
      strikeBrush {
       bIsDynamicallyLoaded=false
       drawAs=Image
       imageType=NoImage
       mirroring=NoMirror
       resourceName=None
       resourceObject=null
       tiling=NoTile
       imageSize=<32 32>
       margin=<ltrb 0 0 0 0>
       outlineSettings {
        bUseBrushTransparency=false
        roundingType=HalfHeightRadius
        width=0
        color {colorUseRule=UseColor_Specified, specifiedColor=<rgba 0 0 0 0>}
        cornerRadii=<0 0 0 0>
       }
       tintColor {colorUseRule=UseColor_Specified, specifiedColor=<rgba 1 1 1 1>}
       uVRegion {bIsValid=false, max=<0 0>, min=<0 0>}
      }
      underlineBrush {
       bIsDynamicallyLoaded=false
       drawAs=Image
       imageType=NoImage
       mirroring=NoMirror
       resourceName=None
       resourceObject=null
       tiling=NoTile
       imageSize=<32 32>
       margin=<ltrb 0 0 0 0>
       outlineSettings {
        bUseBrushTransparency=false
        roundingType=HalfHeightRadius
        width=0
        color {colorUseRule=UseColor_Specified, specifiedColor=<rgba 0 0 0 0>}
        cornerRadii=<0 0 0 0>
       }
       tintColor {colorUseRule=UseColor_Specified, specifiedColor=<rgba 1 1 1 1>}
       uVRegion {bIsValid=false, max=<0 0>, min=<0 0>}
      }
     }
     vScrollBarPadding=<ltrb 0 0 0 0>
    }
   }
   slot {class=VerticalBoxSlot, properties {padding=<ltrb 4 4 4 4>, size {sizeRule=Fill, value=1}}}
  }
  SearchBox_Author {
   widgetClassPath=&EditorUtilityEditableTextBox
   bIsVariable=true
   parent=VB_Advanced
   widget=SearchBox_Author
   properties {
    bOverride_Cursor=true
    clipping=ClipToBounds
    hintText=Search by author...
    minimumDesiredWidth=300
    toolTipText=Needs at least 3 characters
    widgetStyle {
     backgroundColor {colorUseRule=UseColor_Specified, specifiedColor=<rgba 1 1 1 1>}
     backgroundImageFocused {
      bIsDynamicallyLoaded=false
      drawAs=RoundedBox
      imageType=NoImage
      mirroring=NoMirror
      resourceName=None
      resourceObject=null
      tiling=NoTile
      imageSize=<0 0>
      margin=<ltrb 0 0 0 0>
      outlineSettings {
       bUseBrushTransparency=false
       roundingType=FixedRadius
       width=1
       color {colorUseRule=UseColor_ColorTable, specifiedColor=<rgba 0 0 0 0>}
       cornerRadii=<4 4 4 4>
      }
      tintColor {colorUseRule=UseColor_Specified, specifiedColor=<rgba 0.0048 0.0048 0.0048 1>}
      uVRegion {bIsValid=false, max=<0 0>, min=<0 0>}
     }
     backgroundImageHovered {
      bIsDynamicallyLoaded=false
      drawAs=RoundedBox
      imageType=NoImage
      mirroring=NoMirror
      resourceName=None
      resourceObject=null
      tiling=NoTile
      imageSize=<0 0>
      margin=<ltrb 0 0 0 0>
      outlineSettings {
       bUseBrushTransparency=false
       roundingType=FixedRadius
       width=1
       color {colorUseRule=UseColor_ColorTable, specifiedColor=<rgba 0 0 0 0>}
       cornerRadii=<4 4 4 4>
      }
      tintColor {colorUseRule=UseColor_Specified, specifiedColor=<rgba 0.0048 0.0048 0.0048 1>}
      uVRegion {bIsValid=false, max=<0 0>, min=<0 0>}
     }
     backgroundImageNormal {
      bIsDynamicallyLoaded=false
      drawAs=RoundedBox
      imageType=NoImage
      mirroring=NoMirror
      resourceName=None
      resourceObject=null
      tiling=NoTile
      imageSize=<0 0>
      margin=<ltrb 0 0 0 0>
      outlineSettings {
       bUseBrushTransparency=false
       roundingType=FixedRadius
       width=1
       color {colorUseRule=UseColor_ColorTable, specifiedColor=<rgba 0 0 0 0>}
       cornerRadii=<4 4 4 4>
      }
      tintColor {colorUseRule=UseColor_Specified, specifiedColor=<rgba 0.0048 0.0048 0.0048 1>}
      uVRegion {bIsValid=false, max=<0 0>, min=<0 0>}
     }
     backgroundImageReadOnly {
      bIsDynamicallyLoaded=false
      drawAs=RoundedBox
      imageType=NoImage
      mirroring=NoMirror
      resourceName=None
      resourceObject=null
      tiling=NoTile
      imageSize=<0 0>
      margin=<ltrb 0 0 0 0>
      outlineSettings {
       bUseBrushTransparency=false
       roundingType=FixedRadius
       width=1
       color {colorUseRule=UseColor_ColorTable, specifiedColor=<rgba 0 0 0 0>}
       cornerRadii=<4 4 4 4>
      }
      tintColor {colorUseRule=UseColor_Specified, specifiedColor=<rgba 0.0284 0.0284 0.0284 1>}
      uVRegion {bIsValid=false, max=<0 0>, min=<0 0>}
     }
     focusedForegroundColor {colorUseRule=UseColor_Specified, specifiedColor=<rgba 1 1 1 1>}
     foregroundColor {colorUseRule=UseColor_Specified, specifiedColor=<rgba 0.5271 0.5271 0.5271 1>}
     hScrollBarPadding=<ltrb 0 0 0 0>
     padding=<ltrb 4 4 4 4>
     readOnlyForegroundColor {colorUseRule=UseColor_Foreground, specifiedColor=<rgba 1 0 1 1>}
     scrollBarStyle {
      thickness=8
      draggedThumbImage {
       bIsDynamicallyLoaded=false
       drawAs=RoundedBox
       imageType=NoImage
       mirroring=NoMirror
       resourceName=None
       resourceObject=null
       tiling=NoTile
       imageSize=<0 0>
       margin=<ltrb 0 0 0 0>
       outlineSettings {
        bUseBrushTransparency=false
        roundingType=FixedRadius
        width=0
        color {colorUseRule=UseColor_Specified, specifiedColor=<rgba 0 0 0 0>}
        cornerRadii=<4 4 4 4>
       }
       tintColor {colorUseRule=UseColor_Specified, specifiedColor=<rgba 0.2159 0.2159 0.2159 1>}
       uVRegion {bIsValid=false, max=<0 0>, min=<0 0>}
      }
      horizontalBackgroundImage {
       bIsDynamicallyLoaded=false
       drawAs=NoDrawType
       imageType=NoImage
       mirroring=NoMirror
       resourceName=None
       resourceObject=null
       tiling=NoTile
       imageSize=<0 0>
       margin=<ltrb 0 0 0 0>
       outlineSettings {
        bUseBrushTransparency=false
        roundingType=HalfHeightRadius
        width=0
        color {colorUseRule=UseColor_Specified, specifiedColor=<rgba 0 0 0 0>}
        cornerRadii=<0 0 0 0>
       }
       tintColor {colorUseRule=UseColor_Specified, specifiedColor=<rgba 1 1 1 1>}
       uVRegion {bIsValid=false, max=<0 0>, min=<0 0>}
      }
      horizontalBottomSlotImage {
       bIsDynamicallyLoaded=false
       drawAs=NoDrawType
       imageType=NoImage
       mirroring=NoMirror
       resourceName=None
       resourceObject=null
       tiling=NoTile
       imageSize=<0 0>
       margin=<ltrb 0 0 0 0>
       outlineSettings {
        bUseBrushTransparency=false
        roundingType=HalfHeightRadius
        width=0
        color {colorUseRule=UseColor_Specified, specifiedColor=<rgba 0 0 0 0>}
        cornerRadii=<0 0 0 0>
       }
       tintColor {colorUseRule=UseColor_Specified, specifiedColor=<rgba 1 1 1 1>}
       uVRegion {bIsValid=false, max=<0 0>, min=<0 0>}
      }
      horizontalTopSlotImage {
       bIsDynamicallyLoaded=false
       drawAs=NoDrawType
       imageType=NoImage
       mirroring=NoMirror
       resourceName=None
       resourceObject=null
       tiling=NoTile
       imageSize=<0 0>
       margin=<ltrb 0 0 0 0>
       outlineSettings {
        bUseBrushTransparency=false
        roundingType=HalfHeightRadius
        width=0
        color {colorUseRule=UseColor_Specified, specifiedColor=<rgba 0 0 0 0>}
        cornerRadii=<0 0 0 0>
       }
       tintColor {colorUseRule=UseColor_Specified, specifiedColor=<rgba 1 1 1 1>}
       uVRegion {bIsValid=false, max=<0 0>, min=<0 0>}
      }
      hoveredThumbImage {
       bIsDynamicallyLoaded=false
       drawAs=RoundedBox
       imageType=NoImage
       mirroring=NoMirror
       resourceName=None
       resourceObject=null
       tiling=NoTile
       imageSize=<0 0>
       margin=<ltrb 0 0 0 0>
       outlineSettings {
        bUseBrushTransparency=false
        roundingType=FixedRadius
        width=0
        color {colorUseRule=UseColor_Specified, specifiedColor=<rgba 0 0 0 0>}
        cornerRadii=<4 4 4 4>
       }
       tintColor {colorUseRule=UseColor_Specified, specifiedColor=<rgba 0.2159 0.2159 0.2159 1>}
       uVRegion {bIsValid=false, max=<0 0>, min=<0 0>}
      }
      normalThumbImage {
       bIsDynamicallyLoaded=false
       drawAs=RoundedBox
       imageType=NoImage
       mirroring=NoMirror
       resourceName=None
       resourceObject=null
       tiling=NoTile
       imageSize=<0 0>
       margin=<ltrb 0 0 0 0>
       outlineSettings {
        bUseBrushTransparency=false
        roundingType=FixedRadius
        width=0
        color {colorUseRule=UseColor_Specified, specifiedColor=<rgba 0 0 0 0>}
        cornerRadii=<4 4 4 4>
       }
       tintColor {colorUseRule=UseColor_Specified, specifiedColor=<rgba 0.0953 0.0953 0.0953 1>}
       uVRegion {bIsValid=false, max=<0 0>, min=<0 0>}
      }
      verticalBackgroundImage {
       bIsDynamicallyLoaded=false
       drawAs=NoDrawType
       imageType=NoImage
       mirroring=NoMirror
       resourceName=None
       resourceObject=null
       tiling=NoTile
       imageSize=<0 0>
       margin=<ltrb 0 0 0 0>
       outlineSettings {
        bUseBrushTransparency=false
        roundingType=HalfHeightRadius
        width=0
        color {colorUseRule=UseColor_Specified, specifiedColor=<rgba 0 0 0 0>}
        cornerRadii=<0 0 0 0>
       }
       tintColor {colorUseRule=UseColor_Specified, specifiedColor=<rgba 1 1 1 1>}
       uVRegion {bIsValid=false, max=<0 0>, min=<0 0>}
      }
      verticalBottomSlotImage {
       bIsDynamicallyLoaded=false
       drawAs=NoDrawType
       imageType=NoImage
       mirroring=NoMirror
       resourceName=None
       resourceObject=null
       tiling=NoTile
       imageSize=<0 0>
       margin=<ltrb 0 0 0 0>
       outlineSettings {
        bUseBrushTransparency=false
        roundingType=HalfHeightRadius
        width=0
        color {colorUseRule=UseColor_Specified, specifiedColor=<rgba 0 0 0 0>}
        cornerRadii=<0 0 0 0>
       }
       tintColor {colorUseRule=UseColor_Specified, specifiedColor=<rgba 1 1 1 1>}
       uVRegion {bIsValid=false, max=<0 0>, min=<0 0>}
      }
      verticalTopSlotImage {
       bIsDynamicallyLoaded=false
       drawAs=NoDrawType
       imageType=NoImage
       mirroring=NoMirror
       resourceName=None
       resourceObject=null
       tiling=NoTile
       imageSize=<0 0>
       margin=<ltrb 0 0 0 0>
       outlineSettings {
        bUseBrushTransparency=false
        roundingType=HalfHeightRadius
        width=0
        color {colorUseRule=UseColor_Specified, specifiedColor=<rgba 0 0 0 0>}
        cornerRadii=<0 0 0 0>
       }
       tintColor {colorUseRule=UseColor_Specified, specifiedColor=<rgba 1 1 1 1>}
       uVRegion {bIsValid=false, max=<0 0>, min=<0 0>}
      }
     }
     textStyle {
      overflowPolicy=Clip
      transformPolicy=None
      colorAndOpacity {colorUseRule=UseColor_Foreground, specifiedColor=<rgba 1 0 1 1>}
      font {
       bForceMonospaced=false
       bMaterialIsStencil=false
       fontMaterial=null
       fontObject=&Roboto
       letterSpacing=0
       monospacedWidth=1
       size=12
       skewAmount=0
       typefaceFontName=Regular
       outlineSettings {
        bApplyOutlineToDropShadows=false
        bMiteredCorners=false
        bSeparateFillAlpha=false
        outlineMaterial=null
        outlineSize=0
        outlineColor=<rgba 0 0 0 1>
       }
      }
      highlightColor {colorUseRule=UseColor_ColorTable, specifiedColor=<rgba 0 0 0 0>}
      highlightShape {
       bIsDynamicallyLoaded=false
       drawAs=Image
       imageType=NoImage
       mirroring=NoMirror
       resourceName=None
       resourceObject=null
       tiling=NoTile
       imageSize=<0 0>
       margin=<ltrb 0 0 0 0>
       outlineSettings {
        bUseBrushTransparency=false
        roundingType=HalfHeightRadius
        width=0
        color {colorUseRule=UseColor_Specified, specifiedColor=<rgba 0 0 0 0>}
        cornerRadii=<0 0 0 0>
       }
       tintColor {colorUseRule=UseColor_ColorTable, specifiedColor=<rgba 0 0 0 0>}
       uVRegion {bIsValid=false, max=<0 0>, min=<0 0>}
      }
      selectedBackgroundColor {colorUseRule=UseColor_ColorTable, specifiedColor=<rgba 0 0 0 0>}
      shadowColorAndOpacity=<rgba 0 0 0 1>
      shadowOffset=<0 0>
      strikeBrush {
       bIsDynamicallyLoaded=false
       drawAs=Image
       imageType=NoImage
       mirroring=NoMirror
       resourceName=None
       resourceObject=null
       tiling=NoTile
       imageSize=<32 32>
       margin=<ltrb 0 0 0 0>
       outlineSettings {
        bUseBrushTransparency=false
        roundingType=HalfHeightRadius
        width=0
        color {colorUseRule=UseColor_Specified, specifiedColor=<rgba 0 0 0 0>}
        cornerRadii=<0 0 0 0>
       }
       tintColor {colorUseRule=UseColor_Specified, specifiedColor=<rgba 1 1 1 1>}
       uVRegion {bIsValid=false, max=<0 0>, min=<0 0>}
      }
      underlineBrush {
       bIsDynamicallyLoaded=false
       drawAs=Image
       imageType=NoImage
       mirroring=NoMirror
       resourceName=None
       resourceObject=null
       tiling=NoTile
       imageSize=<32 32>
       margin=<ltrb 0 0 0 0>
       outlineSettings {
        bUseBrushTransparency=false
        roundingType=HalfHeightRadius
        width=0
        color {colorUseRule=UseColor_Specified, specifiedColor=<rgba 0 0 0 0>}
        cornerRadii=<0 0 0 0>
       }
       tintColor {colorUseRule=UseColor_Specified, specifiedColor=<rgba 1 1 1 1>}
       uVRegion {bIsValid=false, max=<0 0>, min=<0 0>}
      }
     }
     vScrollBarPadding=<ltrb 0 0 0 0>
    }
   }
   slot {class=VerticalBoxSlot, properties {padding=<ltrb 4 4 4 4>, size {sizeRule=Fill, value=1}}}
  }
  chk_with_all {
   widgetClassPath=&EditorUtilityCheckBox
   bIsVariable=true
   parent=VB_Advanced
   widget=chk_with_all
   properties {toolTipText="When checked,  an asset will be match if all filters must match (instead of any filter can match)"}
   slot {class=VerticalBoxSlot, properties {verticalAlignment=VAlign_Center, padding=<ltrb 4 4 4 4>}}
  }
  TextBlock_71 {
   widgetClassPath=&TextBlock
   parent=chk_with_all
   widget=TextBlock_71
   properties {
    text=Match all
    font {
     bForceMonospaced=false
     bMaterialIsStencil=false
     fontMaterial=null
     fontObject=&Roboto
     letterSpacing=0
     monospacedWidth=1
     size=10.5
     skewAmount=0
     typefaceFontName=Regular
     outlineSettings {
      bApplyOutlineToDropShadows=false
      bMiteredCorners=false
      bSeparateFillAlpha=false
      outlineMaterial=null
      outlineSize=0
      outlineColor=<rgba 0 0 0 1>
     }
    }
   }
   slot {class=PanelSlot}
  }
  chk_filterbyrigs {
   widgetClassPath=&EditorUtilityCheckBox
   bIsVariable=true
   parent=VB_Advanced
   widget=chk_filterbyrigs
   properties {toolTipText=Show only compatible assets with the rig}
   slot {class=VerticalBoxSlot, properties {verticalAlignment=VAlign_Center, padding=<ltrb 4 4 4 4>}}
  }
  Filterbyrigs {
   widgetClassPath=&TextBlock
   parent=chk_filterbyrigs
   widget=Filterbyrigs
   properties {
    text=Filter by rigs in sequencer
    font {
     bForceMonospaced=false
     bMaterialIsStencil=false
     fontMaterial=null
     fontObject=&Roboto
     letterSpacing=0
     monospacedWidth=1
     size=10.5
     skewAmount=0
     typefaceFontName=Regular
     outlineSettings {
      bApplyOutlineToDropShadows=false
      bMiteredCorners=false
      bSeparateFillAlpha=false
      outlineMaterial=null
      outlineSize=0
      outlineColor=<rgba 0 0 0 1>
     }
    }
   }
   slot {class=PanelSlot}
  }
  SearchParamsView {
   widgetClassPath=&DetailsView
   parent=VB_Advanced
   widget=SearchParamsView
   properties {categoriesToShow [Search]}
   slot {class=VerticalBoxSlot, properties {padding=<ltrb 4 4 4 4>}}
  }
  Border_0 {
   widgetClassPath=&Border
   parent=Main
   widget=Border_0
   properties {background {
     bIsDynamicallyLoaded=false
     drawAs=RoundedBox
     imageType=NoImage
     mirroring=NoMirror
     resourceName=None
     resourceObject=null
     tiling=NoTile
     imageSize=<32 32>
     margin=<ltrb 0 0 0 0>
     outlineSettings {
      bUseBrushTransparency=false
      roundingType=FixedRadius
      width=1
      color {colorUseRule=UseColor_Specified, specifiedColor=<rgba 1 1 1 0.5>}
      cornerRadii=<4 4 4 4>
     }
     tintColor {colorUseRule=UseColor_Specified, specifiedColor=<rgba 1 1 1 0>}
     uVRegion {bIsValid=false, max=<0 0>, min=<0 0>}
    }}
   slot {class=VerticalBoxSlot, properties {padding=<ltrb 4 4 4 4>, size {sizeRule=Fill, value=10}}}
  }
  VerticalBox_0 {widgetClassPath=&VerticalBox, parent=Border_0, widget=VerticalBox_0, slot {class=BorderSlot}}
  Assets {
   widgetClassPath=/Script/UMG.TileView
   bIsVariable=true
   parent=VerticalBox_0
   widget=Assets
   properties {
    bClearSelectionOnClick=true
    bEnableTouchScrolling=false
    bEntrySizeIncludesEntrySpacing=false
    entryHeight=256
    entryWidgetClass=/AnimationLibrary/BP/EWC_AnimEntry.EWC_AnimEntry_C
    entryWidth=256
    horizontalEntrySpacing=4
    tileAlignment=LeftAligned
    verticalEntrySpacing=4
    scrollBarStyle {
     thickness=8
     draggedThumbImage {
      bIsDynamicallyLoaded=false
      drawAs=RoundedBox
      imageType=NoImage
      mirroring=NoMirror
      resourceName=None
      resourceObject=null
      tiling=NoTile
      imageSize=<0 0>
      margin=<ltrb 0 0 0 0>
      outlineSettings {
       bUseBrushTransparency=false
       roundingType=FixedRadius
       width=0
       color {colorUseRule=UseColor_Specified, specifiedColor=<rgba 0 0 0 0>}
       cornerRadii=<4 4 4 4>
      }
      tintColor {colorUseRule=UseColor_Specified, specifiedColor=<rgba 0.2159 0.2159 0.2159 1>}
      uVRegion {bIsValid=false, max=<0 0>, min=<0 0>}
     }
     horizontalBackgroundImage {
      bIsDynamicallyLoaded=false
      drawAs=NoDrawType
      imageType=NoImage
      mirroring=NoMirror
      resourceName=None
      resourceObject=null
      tiling=NoTile
      imageSize=<0 0>
      margin=<ltrb 0 0 0 0>
      outlineSettings {
       bUseBrushTransparency=false
       roundingType=HalfHeightRadius
       width=0
       color {colorUseRule=UseColor_Specified, specifiedColor=<rgba 0 0 0 0>}
       cornerRadii=<0 0 0 0>
      }
      tintColor {colorUseRule=UseColor_Specified, specifiedColor=<rgba 1 1 1 1>}
      uVRegion {bIsValid=false, max=<0 0>, min=<0 0>}
     }
     horizontalBottomSlotImage {
      bIsDynamicallyLoaded=false
      drawAs=NoDrawType
      imageType=NoImage
      mirroring=NoMirror
      resourceName=None
      resourceObject=null
      tiling=NoTile
      imageSize=<0 0>
      margin=<ltrb 0 0 0 0>
      outlineSettings {
       bUseBrushTransparency=false
       roundingType=HalfHeightRadius
       width=0
       color {colorUseRule=UseColor_Specified, specifiedColor=<rgba 0 0 0 0>}
       cornerRadii=<0 0 0 0>
      }
      tintColor {colorUseRule=UseColor_Specified, specifiedColor=<rgba 1 1 1 1>}
      uVRegion {bIsValid=false, max=<0 0>, min=<0 0>}
     }
     horizontalTopSlotImage {
      bIsDynamicallyLoaded=false
      drawAs=NoDrawType
      imageType=NoImage
      mirroring=NoMirror
      resourceName=None
      resourceObject=null
      tiling=NoTile
      imageSize=<0 0>
      margin=<ltrb 0 0 0 0>
      outlineSettings {
       bUseBrushTransparency=false
       roundingType=HalfHeightRadius
       width=0
       color {colorUseRule=UseColor_Specified, specifiedColor=<rgba 0 0 0 0>}
       cornerRadii=<0 0 0 0>
      }
      tintColor {colorUseRule=UseColor_Specified, specifiedColor=<rgba 1 1 1 1>}
      uVRegion {bIsValid=false, max=<0 0>, min=<0 0>}
     }
     hoveredThumbImage {
      bIsDynamicallyLoaded=false
      drawAs=RoundedBox
      imageType=NoImage
      mirroring=NoMirror
      resourceName=None
      resourceObject=null
      tiling=NoTile
      imageSize=<0 0>
      margin=<ltrb 0 0 0 0>
      outlineSettings {
       bUseBrushTransparency=false
       roundingType=FixedRadius
       width=0
       color {colorUseRule=UseColor_Specified, specifiedColor=<rgba 0 0 0 0>}
       cornerRadii=<4 4 4 4>
      }
      tintColor {colorUseRule=UseColor_Specified, specifiedColor=<rgba 0.2159 0.2159 0.2159 1>}
      uVRegion {bIsValid=false, max=<0 0>, min=<0 0>}
     }
     normalThumbImage {
      bIsDynamicallyLoaded=false
      drawAs=RoundedBox
      imageType=NoImage
      mirroring=NoMirror
      resourceName=None
      resourceObject=null
      tiling=NoTile
      imageSize=<0 0>
      margin=<ltrb 0 0 0 0>
      outlineSettings {
       bUseBrushTransparency=false
       roundingType=FixedRadius
       width=0
       color {colorUseRule=UseColor_Specified, specifiedColor=<rgba 0 0 0 0>}
       cornerRadii=<4 4 4 4>
      }
      tintColor {colorUseRule=UseColor_Specified, specifiedColor=<rgba 0.0953 0.0953 0.0953 1>}
      uVRegion {bIsValid=false, max=<0 0>, min=<0 0>}
     }
     verticalBackgroundImage {
      bIsDynamicallyLoaded=false
      drawAs=NoDrawType
      imageType=NoImage
      mirroring=NoMirror
      resourceName=None
      resourceObject=null
      tiling=NoTile
      imageSize=<0 0>
      margin=<ltrb 0 0 0 0>
      outlineSettings {
       bUseBrushTransparency=false
       roundingType=HalfHeightRadius
       width=0
       color {colorUseRule=UseColor_Specified, specifiedColor=<rgba 0 0 0 0>}
       cornerRadii=<0 0 0 0>
      }
      tintColor {colorUseRule=UseColor_Specified, specifiedColor=<rgba 1 1 1 1>}
      uVRegion {bIsValid=false, max=<0 0>, min=<0 0>}
     }
     verticalBottomSlotImage {
      bIsDynamicallyLoaded=false
      drawAs=NoDrawType
      imageType=NoImage
      mirroring=NoMirror
      resourceName=None
      resourceObject=null
      tiling=NoTile
      imageSize=<0 0>
      margin=<ltrb 0 0 0 0>
      outlineSettings {
       bUseBrushTransparency=false
       roundingType=HalfHeightRadius
       width=0
       color {colorUseRule=UseColor_Specified, specifiedColor=<rgba 0 0 0 0>}
       cornerRadii=<0 0 0 0>
      }
      tintColor {colorUseRule=UseColor_Specified, specifiedColor=<rgba 1 1 1 1>}
      uVRegion {bIsValid=false, max=<0 0>, min=<0 0>}
     }
     verticalTopSlotImage {
      bIsDynamicallyLoaded=false
      drawAs=NoDrawType
      imageType=NoImage
      mirroring=NoMirror
      resourceName=None
      resourceObject=null
      tiling=NoTile
      imageSize=<0 0>
      margin=<ltrb 0 0 0 0>
      outlineSettings {
       bUseBrushTransparency=false
       roundingType=HalfHeightRadius
       width=0
       color {colorUseRule=UseColor_Specified, specifiedColor=<rgba 0 0 0 0>}
       cornerRadii=<0 0 0 0>
      }
      tintColor {colorUseRule=UseColor_Specified, specifiedColor=<rgba 1 1 1 1>}
      uVRegion {bIsValid=false, max=<0 0>, min=<0 0>}
     }
    }
   }
   slot {class=VerticalBoxSlot, properties {size {sizeRule=Fill, value=1}}}
  }
  AssetsAsList {
   widgetClassPath=/Script/Blutility.EditorUtilityListView
   bIsVariable=true
   parent=VerticalBox_0
   widget=AssetsAsList
   properties {entryWidgetClass=&EWC_AnimEntry_C, verticalEntrySpacing=4, visibility=Collapsed}
   slot {class=VerticalBoxSlot, properties {size {sizeRule=Fill, value=1}}}
  }
  Footer {
   widgetClassPath=&WrapBox
   parent=Main
   widget=Footer
   properties {bOverride_Cursor=true, horizontalAlignment=HAlign_Fill}
   slot {class=VerticalBoxSlot, properties {padding=<ltrb 4 0 4 4>, size {sizeRule=Automatic, value=0.1}}}
  }
  BorderPaste {
   widgetClassPath=&Border
   parent=Footer
   widget=BorderPaste
   properties {
    background {
     bIsDynamicallyLoaded=false
     drawAs=NoDrawType
     imageType=NoImage
     mirroring=NoMirror
     resourceName=None
     resourceObject=null
     tiling=NoTile
     imageSize=<32 32>
     margin=<ltrb 0 0 0 0>
     outlineSettings {
      bUseBrushTransparency=false
      roundingType=FixedRadius
      width=1
      color {colorUseRule=UseColor_Specified, specifiedColor=<rgba 1 1 1 0.5>}
      cornerRadii=<4 4 4 4>
     }
     tintColor {colorUseRule=UseColor_Specified, specifiedColor=<rgba 1 1 1 0>}
     uVRegion {bIsValid=false, max=<0 0>, min=<0 0>}
    }
    padding=<ltrb 4 4 4 4>
   }
   slot {class=WrapBoxSlot}
  }
  WrapBox_1 {widgetClassPath=&WrapBox, parent=BorderPaste, widget=WrapBox_1, slot {class=BorderSlot, properties {padding=<ltrb 4 4 4 4>}}}
  PasteActions {
   widgetClassPath=&DropdownButton
   bIsVariable=true
   parent=WrapBox_1
   widget=PasteActions
   properties {
    mainButtonText=Apply
    font {
     bForceMonospaced=false
     bMaterialIsStencil=false
     fontMaterial=null
     fontObject=&Roboto
     letterSpacing=0
     monospacedWidth=1
     size=9
     skewAmount=0
     typefaceFontName=Bold
     outlineSettings {
      bApplyOutlineToDropShadows=false
      bMiteredCorners=false
      bSeparateFillAlpha=false
      outlineMaterial=null
      outlineSize=0
      outlineColor=<rgba 0 0 0 1>
     }
    }
    foregroundColor {colorUseRule=UseColor_Specified, specifiedColor=<rgba 0.5029 0.0844 0.0865 1>}
   }
   slot {class=WrapBoxSlot, properties {horizontalAlignment=HAlign_Left, verticalAlignment=VAlign_Top}}
  }
  BorderCopy {
   widgetClassPath=&Border
   parent=Footer
   widget=BorderCopy
   properties {
    bIsEnabled=false
    visibility=Collapsed
    background {
     bIsDynamicallyLoaded=false
     drawAs=NoDrawType
     imageType=NoImage
     mirroring=NoMirror
     resourceName=None
     resourceObject=null
     tiling=NoTile
     imageSize=<32 32>
     margin=<ltrb 0 0 0 0>
     outlineSettings {
      bUseBrushTransparency=false
      roundingType=FixedRadius
      width=1
      color {colorUseRule=UseColor_Specified, specifiedColor=<rgba 1 1 1 0.5>}
      cornerRadii=<4 4 4 4>
     }
     tintColor {colorUseRule=UseColor_Specified, specifiedColor=<rgba 1 1 1 0>}
     uVRegion {bIsValid=false, max=<0 0>, min=<0 0>}
    }
    padding=<ltrb 4 4 4 4>
   }
   slot {class=WrapBoxSlot}
  }
  WrapBox_2 {
   widgetClassPath=&WrapBox
   parent=BorderCopy
   widget=WrapBox_2
   properties {bOverride_Cursor=true}
   slot {class=BorderSlot, properties {padding=<ltrb 4 4 4 4>}}
  }
  HorizontalBox_1 {
   widgetClassPath=&HorizontalBox
   parent=Footer
   widget=HorizontalBox_1
   properties {bIsEnabled=false, toolTipText="[WIP] Coming Back soon™"}
   slot {class=WrapBoxSlot}
  }
  TextBlock {
   widgetClassPath=&TextBlock
   parent=HorizontalBox_1
   widget=TextBlock
   properties {
    lineHeightPercentage=0.94
    text=Sets
    colorAndOpacity {colorUseRule=UseColor_Specified, specifiedColor=<rgba 1 1 1 0.5>}
    font {
     bForceMonospaced=false
     bMaterialIsStencil=false
     fontMaterial=null
     fontObject=&Roboto
     letterSpacing=0
     monospacedWidth=1
     size=12
     skewAmount=0
     typefaceFontName=Bold
     outlineSettings {
      bApplyOutlineToDropShadows=false
      bMiteredCorners=false
      bSeparateFillAlpha=false
      outlineMaterial=null
      outlineSize=0
      outlineColor=<rgba 0 0 0 1>
     }
    }
   }
   slot {class=HorizontalBoxSlot, properties {verticalAlignment=VAlign_Center, padding=<ltrb 8 8 8 8>}}
  }
  cbSelectionSet {
   widgetClassPath=/Script/FortniteEditor.SearchableStringComboBox
   bIsVariable=true
   parent=HorizontalBox_1
   widget=cbSelectionSet
   properties {
    foregroundColor {colorUseRule=UseColor_Specified, specifiedColor=<rgba 0.5271 0.5271 0.5271 1>}
    itemStyle {
     bUseParentRowBrush=false
     activeBrush {
      bIsDynamicallyLoaded=false
      drawAs=Image
      imageType=NoImage
      mirroring=NoMirror
      resourceName=None
      resourceObject=null
      tiling=NoTile
      imageSize=<0 0>
      margin=<ltrb 0 0 0 0>
      outlineSettings {
       bUseBrushTransparency=false
       roundingType=HalfHeightRadius
       width=0
       color {colorUseRule=UseColor_Specified, specifiedColor=<rgba 0 0 0 0>}
       cornerRadii=<0 0 0 0>
      }
      tintColor {colorUseRule=UseColor_Specified, specifiedColor=<rgba 0 0.162 0.7454 1>}
      uVRegion {bIsValid=false, max=<0 0>, min=<0 0>}
     }
     activeHighlightedBrush {
      bIsDynamicallyLoaded=false
      drawAs=Image
      imageType=NoImage
      mirroring=NoMirror
      resourceName=None
      resourceObject=null
      tiling=NoTile
      imageSize=<0 0>
      margin=<ltrb 0 0 0 0>
      outlineSettings {
       bUseBrushTransparency=false
       roundingType=HalfHeightRadius
       width=0
       color {colorUseRule=UseColor_Specified, specifiedColor=<rgba 0 0 0 0>}
       cornerRadii=<0 0 0 0>
      }
      tintColor {colorUseRule=UseColor_Specified, specifiedColor=<rgba 0.0252 0.0319 0.0423 1>}
      uVRegion {bIsValid=false, max=<0 0>, min=<0 0>}
     }
     activeHoveredBrush {
      bIsDynamicallyLoaded=false
      drawAs=Image
      imageType=NoImage
      mirroring=NoMirror
      resourceName=None
      resourceObject=null
      tiling=NoTile
      imageSize=<0 0>
      margin=<ltrb 0 0 0 0>
      outlineSettings {
       bUseBrushTransparency=false
       roundingType=HalfHeightRadius
       width=0
       color {colorUseRule=UseColor_Specified, specifiedColor=<rgba 0 0 0 0>}
       cornerRadii=<0 0 0 0>
      }
      tintColor {colorUseRule=UseColor_Specified, specifiedColor=<rgba 0 0.162 0.7454 1>}
      uVRegion {bIsValid=false, max=<0 0>, min=<0 0>}
     }
     dropIndicator_Above {
      bIsDynamicallyLoaded=false
      drawAs=Box
      imageType=FullColor
      mirroring=NoMirror
      resourceName=../../../Engine/Content/Slate/Common/DropZoneIndicator_Above.png
      resourceObject=null
      tiling=NoTile
      imageSize=<0 0>
      margin=<ltrb 0.625 0.625 0 0>
      outlineSettings {
       bUseBrushTransparency=false
       roundingType=HalfHeightRadius
       width=0
       color {colorUseRule=UseColor_Specified, specifiedColor=<rgba 0 0 0 0>}
       cornerRadii=<0 0 0 0>
      }
      tintColor {colorUseRule=UseColor_Specified, specifiedColor=<rgba 0.0091 0.3515 0.9647 1>}
      uVRegion {bIsValid=false, max=<0 0>, min=<0 0>}
     }
     dropIndicator_Below {
      bIsDynamicallyLoaded=false
      drawAs=Box
      imageType=FullColor
      mirroring=NoMirror
      resourceName=../../../Engine/Content/Slate/Common/DropZoneIndicator_Below.png
      resourceObject=null
      tiling=NoTile
      imageSize=<0 0>
      margin=<ltrb 0.625 0 0 0.625>
      outlineSettings {
       bUseBrushTransparency=false
       roundingType=HalfHeightRadius
       width=0
       color {colorUseRule=UseColor_Specified, specifiedColor=<rgba 0 0 0 0>}
       cornerRadii=<0 0 0 0>
      }
      tintColor {colorUseRule=UseColor_Specified, specifiedColor=<rgba 0.0091 0.3515 0.9647 1>}
      uVRegion {bIsValid=false, max=<0 0>, min=<0 0>}
     }
     dropIndicator_Onto {
      bIsDynamicallyLoaded=false
      drawAs=Box
      imageType=FullColor
      mirroring=NoMirror
      resourceName=../../../Engine/Content/Slate/Common/DropZoneIndicator_Onto.png
      resourceObject=null
      tiling=NoTile
      imageSize=<0 0>
      margin=<ltrb 0.25 0.25 0.25 0.25>
      outlineSettings {
       bUseBrushTransparency=false
       roundingType=HalfHeightRadius
       width=0
       color {colorUseRule=UseColor_Specified, specifiedColor=<rgba 0 0 0 0>}
       cornerRadii=<0 0 0 0>
      }
      tintColor {colorUseRule=UseColor_Specified, specifiedColor=<rgba 0.0091 0.3515 0.9647 1>}
      uVRegion {bIsValid=false, max=<0 0>, min=<0 0>}
     }
     evenRowBackgroundBrush {
      bIsDynamicallyLoaded=false
      drawAs=Image
      imageType=NoImage
      mirroring=NoMirror
      resourceName=None
      resourceObject=null
      tiling=NoTile
      imageSize=<0 0>
      margin=<ltrb 0 0 0 0>
      outlineSettings {
       bUseBrushTransparency=false
       roundingType=HalfHeightRadius
       width=0
       color {colorUseRule=UseColor_Specified, specifiedColor=<rgba 0 0 0 0>}
       cornerRadii=<0 0 0 0>
      }
      tintColor {colorUseRule=UseColor_Specified, specifiedColor=<rgba 0.0103 0.0103 0.0103 1>}
      uVRegion {bIsValid=false, max=<0 0>, min=<0 0>}
     }
     evenRowBackgroundHoveredBrush {
      bIsDynamicallyLoaded=false
      drawAs=Image
      imageType=NoImage
      mirroring=NoMirror
      resourceName=None
      resourceObject=null
      tiling=NoTile
      imageSize=<0 0>
      margin=<ltrb 0 0 0 0>
      outlineSettings {
       bUseBrushTransparency=false
       roundingType=HalfHeightRadius
       width=0
       color {colorUseRule=UseColor_Specified, specifiedColor=<rgba 0 0 0 0>}
       cornerRadii=<0 0 0 0>
      }
      tintColor {colorUseRule=UseColor_Specified, specifiedColor=<rgba 0.0176 0.0176 0.0176 1>}
      uVRegion {bIsValid=false, max=<0 0>, min=<0 0>}
     }
     inactiveBrush {
      bIsDynamicallyLoaded=false
      drawAs=Image
      imageType=NoImage
      mirroring=NoMirror
      resourceName=None
      resourceObject=null
      tiling=NoTile
      imageSize=<0 0>
      margin=<ltrb 0 0 0 0>
      outlineSettings {
       bUseBrushTransparency=false
       roundingType=HalfHeightRadius
       width=0
       color {colorUseRule=UseColor_Specified, specifiedColor=<rgba 0 0 0 0>}
       cornerRadii=<0 0 0 0>
      }
      tintColor {colorUseRule=UseColor_Specified, specifiedColor=<rgba 0.0513 0.0953 0.159 1>}
      uVRegion {bIsValid=false, max=<0 0>, min=<0 0>}
     }
     inactiveHighlightedBrush {
      bIsDynamicallyLoaded=false
      drawAs=Image
      imageType=NoImage
      mirroring=NoMirror
      resourceName=None
      resourceObject=null
      tiling=NoTile
      imageSize=<0 0>
      margin=<ltrb 0 0 0 0>
      outlineSettings {
       bUseBrushTransparency=false
       roundingType=HalfHeightRadius
       width=0
       color {colorUseRule=UseColor_Specified, specifiedColor=<rgba 0 0 0 0>}
       cornerRadii=<0 0 0 0>
      }
      tintColor {colorUseRule=UseColor_Specified, specifiedColor=<rgba 0.0252 0.0319 0.0423 1>}
      uVRegion {bIsValid=false, max=<0 0>, min=<0 0>}
     }
     inactiveHoveredBrush {
      bIsDynamicallyLoaded=false
      drawAs=Image
      imageType=NoImage
      mirroring=NoMirror
      resourceName=None
      resourceObject=null
      tiling=NoTile
      imageSize=<0 0>
      margin=<ltrb 0 0 0 0>
      outlineSettings {
       bUseBrushTransparency=false
       roundingType=HalfHeightRadius
       width=0
       color {colorUseRule=UseColor_Specified, specifiedColor=<rgba 0 0 0 0>}
       cornerRadii=<0 0 0 0>
      }
      tintColor {colorUseRule=UseColor_Specified, specifiedColor=<rgba 0.0513 0.0953 0.159 1>}
      uVRegion {bIsValid=false, max=<0 0>, min=<0 0>}
     }
     oddRowBackgroundBrush {
      bIsDynamicallyLoaded=false
      drawAs=Image
      imageType=NoImage
      mirroring=NoMirror
      resourceName=None
      resourceObject=null
      tiling=NoTile
      imageSize=<0 0>
      margin=<ltrb 0 0 0 0>
      outlineSettings {
       bUseBrushTransparency=false
       roundingType=HalfHeightRadius
       width=0
       color {colorUseRule=UseColor_Specified, specifiedColor=<rgba 0 0 0 0>}
       cornerRadii=<0 0 0 0>
      }
      tintColor {colorUseRule=UseColor_Specified, specifiedColor=<rgba 0.0103 0.0103 0.0103 1>}
      uVRegion {bIsValid=false, max=<0 0>, min=<0 0>}
     }
     oddRowBackgroundHoveredBrush {
      bIsDynamicallyLoaded=false
      drawAs=Image
      imageType=NoImage
      mirroring=NoMirror
      resourceName=None
      resourceObject=null
      tiling=NoTile
      imageSize=<0 0>
      margin=<ltrb 0 0 0 0>
      outlineSettings {
       bUseBrushTransparency=false
       roundingType=HalfHeightRadius
       width=0
       color {colorUseRule=UseColor_Specified, specifiedColor=<rgba 0 0 0 0>}
       cornerRadii=<0 0 0 0>
      }
      tintColor {colorUseRule=UseColor_Specified, specifiedColor=<rgba 0.0176 0.0176 0.0176 1>}
      uVRegion {bIsValid=false, max=<0 0>, min=<0 0>}
     }
     parentRowBackgroundBrush {
      bIsDynamicallyLoaded=false
      drawAs=Image
      imageType=NoImage
      mirroring=NoMirror
      resourceName=None
      resourceObject=null
      tiling=NoTile
      imageSize=<32 32>
      margin=<ltrb 0 0 0 0>
      outlineSettings {
       bUseBrushTransparency=false
       roundingType=HalfHeightRadius
       width=0
       color {colorUseRule=UseColor_Specified, specifiedColor=<rgba 0 0 0 0>}
       cornerRadii=<0 0 0 0>
      }
      tintColor {colorUseRule=UseColor_Specified, specifiedColor=<rgba 1 1 1 1>}
      uVRegion {bIsValid=false, max=<0 0>, min=<0 0>}
     }
     parentRowBackgroundHoveredBrush {
      bIsDynamicallyLoaded=false
      drawAs=Image
      imageType=NoImage
      mirroring=NoMirror
      resourceName=None
      resourceObject=null
      tiling=NoTile
      imageSize=<32 32>
      margin=<ltrb 0 0 0 0>
      outlineSettings {
       bUseBrushTransparency=false
       roundingType=HalfHeightRadius
       width=0
       color {colorUseRule=UseColor_Specified, specifiedColor=<rgba 0 0 0 0>}
       cornerRadii=<0 0 0 0>
      }
      tintColor {colorUseRule=UseColor_Specified, specifiedColor=<rgba 1 1 1 1>}
      uVRegion {bIsValid=false, max=<0 0>, min=<0 0>}
     }
     selectedTextColor {colorUseRule=UseColor_Specified, specifiedColor=<rgba 1 1 1 1>}
     selectorFocusedBrush {
      bIsDynamicallyLoaded=false
      drawAs=Border
      imageType=FullColor
      mirroring=NoMirror
      resourceName=../../../Engine/Content/Slate/Common/Selector.png
      resourceObject=null
      tiling=Both
      imageSize=<0 0>
      margin=<ltrb 0.25 0.25 0.25 0.25>
      outlineSettings {
       bUseBrushTransparency=false
       roundingType=HalfHeightRadius
       width=0
       color {colorUseRule=UseColor_Specified, specifiedColor=<rgba 0 0 0 0>}
       cornerRadii=<0 0 0 0>
      }
      tintColor {colorUseRule=UseColor_Specified, specifiedColor=<rgba 0.701 0.225 0.003 1>}
      uVRegion {bIsValid=false, max=<0 0>, min=<0 0>}
     }
     textColor {colorUseRule=UseColor_Specified, specifiedColor=<rgba 0.5271 0.5271 0.5271 1>}
    }
    widgetStyle {
     comboButtonStyle {
      downArrowAlign=VAlign_Center
      buttonStyle {
       clickedSlateSound {resourceObject=null}
       disabled {
        bIsDynamicallyLoaded=false
        drawAs=RoundedBox
        imageType=NoImage
        mirroring=NoMirror
        resourceName=None
        resourceObject=null
        tiling=NoTile
        imageSize=<0 0>
        margin=<ltrb 0 0 0 0>
        outlineSettings {
         bUseBrushTransparency=false
         roundingType=FixedRadius
         width=1
         color {colorUseRule=UseColor_ColorTable, specifiedColor=<rgba 0 0 0 0>}
         cornerRadii=<4 4 4 4>
        }
        tintColor {colorUseRule=UseColor_Specified, specifiedColor=<rgba 0.0395 0.0395 0.0395 1>}
        uVRegion {bIsValid=false, max=<0 0>, min=<0 0>}
       }
       disabledForeground {colorUseRule=UseColor_ColorTable, specifiedColor=<rgba 0 0 0 0>}
       hovered {
        bIsDynamicallyLoaded=false
        drawAs=RoundedBox
        imageType=NoImage
        mirroring=NoMirror
        resourceName=None
        resourceObject=null
        tiling=NoTile
        imageSize=<0 0>
        margin=<ltrb 0 0 0 0>
        outlineSettings {
         bUseBrushTransparency=false
         roundingType=FixedRadius
         width=1
         color {colorUseRule=UseColor_ColorTable, specifiedColor=<rgba 0 0 0 0>}
         cornerRadii=<4 4 4 4>
        }
        tintColor {colorUseRule=UseColor_Specified, specifiedColor=<rgba 0.0048 0.0048 0.0048 1>}
        uVRegion {bIsValid=false, max=<0 0>, min=<0 0>}
       }
       hoveredForeground {colorUseRule=UseColor_ColorTable, specifiedColor=<rgba 0 0 0 0>}
       hoveredSlateSound {resourceObject=null}
       normal {
        bIsDynamicallyLoaded=false
        drawAs=RoundedBox
        imageType=NoImage
        mirroring=NoMirror
        resourceName=None
        resourceObject=null
        tiling=NoTile
        imageSize=<0 0>
        margin=<ltrb 0 0 0 0>
        outlineSettings {
         bUseBrushTransparency=false
         roundingType=FixedRadius
         width=1
         color {colorUseRule=UseColor_ColorTable, specifiedColor=<rgba 0 0 0 0>}
         cornerRadii=<4 4 4 4>
        }
        tintColor {colorUseRule=UseColor_Specified, specifiedColor=<rgba 0.0048 0.0048 0.0048 1>}
        uVRegion {bIsValid=false, max=<0 0>, min=<0 0>}
       }
       normalForeground {colorUseRule=UseColor_ColorTable, specifiedColor=<rgba 0 0 0 0>}
       normalPadding=<ltrb 8 1 8 1>
       pressed {
        bIsDynamicallyLoaded=false
        drawAs=RoundedBox
        imageType=NoImage
        mirroring=NoMirror
        resourceName=None
        resourceObject=null
        tiling=NoTile
        imageSize=<0 0>
        margin=<ltrb 0 0 0 0>
        outlineSettings {
         bUseBrushTransparency=false
         roundingType=FixedRadius
         width=1
         color {colorUseRule=UseColor_ColorTable, specifiedColor=<rgba 0 0 0 0>}
         cornerRadii=<4 4 4 4>
        }
        tintColor {colorUseRule=UseColor_Specified, specifiedColor=<rgba 0.0048 0.0048 0.0048 1>}
        uVRegion {bIsValid=false, max=<0 0>, min=<0 0>}
       }
       pressedForeground {colorUseRule=UseColor_ColorTable, specifiedColor=<rgba 0 0 0 0>}
       pressedPadding=<ltrb 8 2 8 0>
       pressedSlateSound {resourceObject=null}
      }
      contentPadding=<ltrb 0 0 0 0>
      downArrowImage {
       bIsDynamicallyLoaded=false
       drawAs=Image
       imageType=Vector
       mirroring=NoMirror
       resourceName=../../../Engine/Content/Slate/Starship/CoreWidgets/ComboBox/wide-chevron-down.svg
       resourceObject=null
       tiling=NoTile
       imageSize=<16 16>
       margin=<ltrb 0 0 0 0>
       outlineSettings {
        bUseBrushTransparency=false
        roundingType=HalfHeightRadius
        width=0
        color {colorUseRule=UseColor_Specified, specifiedColor=<rgba 0 0 0 0>}
        cornerRadii=<0 0 0 0>
       }
       tintColor {colorUseRule=UseColor_Specified, specifiedColor=<rgba 1 1 1 1>}
       uVRegion {bIsValid=false, max=<0 0>, min=<0 0>}
      }
      downArrowPadding=<ltrb 2 2 2 2>
      menuBorderBrush {
       bIsDynamicallyLoaded=false
       drawAs=RoundedBox
       imageType=NoImage
       mirroring=NoMirror
       resourceName=None
       resourceObject=null
       tiling=NoTile
       imageSize=<0 0>
       margin=<ltrb 0 0 0 0>
       outlineSettings {
        bUseBrushTransparency=false
        roundingType=FixedRadius
        width=1
        color {colorUseRule=UseColor_ColorTable, specifiedColor=<rgba 0 0 0 0>}
        cornerRadii=<0 0 0 0>
       }
       tintColor {colorUseRule=UseColor_Specified, specifiedColor=<rgba 0.0395 0.0395 0.0395 1>}
       uVRegion {bIsValid=false, max=<0 0>, min=<0 0>}
      }
      menuBorderPadding=<ltrb 0 0 0 0>
      shadowColorAndOpacity=<rgba 0 0 0 1>
      shadowOffset=<0 0>
     }
     contentPadding=<ltrb 0 0 0 0>
     menuRowPadding=<ltrb 8 3 8 3>
     pressedSlateSound {resourceObject=null}
     selectionChangeSlateSound {resourceObject=null}
    }
   }
   slot {class=HorizontalBoxSlot}
  }
  btnNewSet {
   widgetClassPath=&EditorUtilityButton
   parent=HorizontalBox_1
   widget=btnNewSet
   properties {
    toolTipText="[NotImplemented] Create a new selection Set"
    widgetStyle {
     clickedSlateSound {resourceObject=null}
     disabled {
      bIsDynamicallyLoaded=false
      drawAs=RoundedBox
      imageType=NoImage
      mirroring=NoMirror
      resourceName=None
      resourceObject=null
      tiling=NoTile
      imageSize=<0 0>
      margin=<ltrb 0 0 0 0>
      outlineSettings {
       bUseBrushTransparency=false
       roundingType=FixedRadius
       width=1
       color {colorUseRule=UseColor_ColorTable, specifiedColor=<rgba 0 0 0 0>}
       cornerRadii=<4 4 4 4>
      }
      tintColor {colorUseRule=UseColor_Specified, specifiedColor=<rgba 0.0395 0.0395 0.0395 1>}
      uVRegion {bIsValid=false, max=<0 0>, min=<0 0>}
     }
     disabledForeground {colorUseRule=UseColor_ColorTable, specifiedColor=<rgba 0 0 0 0>}
     hovered {
      bIsDynamicallyLoaded=false
      drawAs=RoundedBox
      imageType=NoImage
      mirroring=NoMirror
      resourceName=None
      resourceObject=null
      tiling=NoTile
      imageSize=<32 32>
      margin=<ltrb 0 0 0 0>
      outlineSettings {
       bUseBrushTransparency=false
       roundingType=FixedRadius
       width=1
       color {colorUseRule=UseColor_ColorTable, specifiedColor=<rgba 0 0 0 0>}
       cornerRadii=<4 4 4 4>
      }
      tintColor {colorUseRule=UseColor_Specified, specifiedColor=<rgba 0.0953 0.0953 0.0953 1>}
      uVRegion {bIsValid=false, max=<0 0>, min=<0 0>}
     }
     hoveredForeground {colorUseRule=UseColor_ColorTable, specifiedColor=<rgba 0 0 0 0>}
     hoveredSlateSound {resourceObject=null}
     normal {
      bIsDynamicallyLoaded=false
      drawAs=RoundedBox
      imageType=NoImage
      mirroring=NoMirror
      resourceName=None
      resourceObject=null
      tiling=NoTile
      imageSize=<32 32>
      margin=<ltrb 0 0 0 0>
      outlineSettings {
       bUseBrushTransparency=false
       roundingType=FixedRadius
       width=1
       color {colorUseRule=UseColor_ColorTable, specifiedColor=<rgba 0 0 0 0>}
       cornerRadii=<4 4 4 4>
      }
      tintColor {colorUseRule=UseColor_Specified, specifiedColor=<rgba 0.0395 0.0395 0.0395 1>}
      uVRegion {bIsValid=false, max=<0 0>, min=<0 0>}
     }
     normalForeground {colorUseRule=UseColor_ColorTable, specifiedColor=<rgba 0 0 0 0>}
     normalPadding=<ltrb 4 4 4 4>
     pressed {
      bIsDynamicallyLoaded=false
      drawAs=RoundedBox
      imageType=NoImage
      mirroring=NoMirror
      resourceName=None
      resourceObject=null
      tiling=NoTile
      imageSize=<32 32>
      margin=<ltrb 0 0 0 0>
      outlineSettings {
       bUseBrushTransparency=false
       roundingType=FixedRadius
       width=1
       color {colorUseRule=UseColor_ColorTable, specifiedColor=<rgba 0 0 0 0>}
       cornerRadii=<4 4 4 4>
      }
      tintColor {colorUseRule=UseColor_Specified, specifiedColor=<rgba 0.0284 0.0284 0.0284 1>}
      uVRegion {bIsValid=false, max=<0 0>, min=<0 0>}
     }
     pressedForeground {colorUseRule=UseColor_ColorTable, specifiedColor=<rgba 0 0 0 0>}
     pressedPadding=<ltrb 4 4.5 4 3.5>
     pressedSlateSound {resourceObject=null}
    }
   }
   slot {class=HorizontalBoxSlot, properties {verticalAlignment=VAlign_Center}}
  }
  imgAddFolderbuttonIcon_1 {
   widgetClassPath=&Image
   parent=btnNewSet
   widget=imgAddFolderbuttonIcon_1
   properties {brush {
     bIsDynamicallyLoaded=false
     drawAs=Image
     imageType=FullColor
     mirroring=NoMirror
     resourceName=None
     resourceObject=/AnimationLibrary/resources/icons/plussolid
     tiling=NoTile
     imageSize=<20 20>
     margin=<ltrb 0 0 0 0>
     outlineSettings {
      bUseBrushTransparency=false
      roundingType=HalfHeightRadius
      width=0
      color {colorUseRule=UseColor_Specified, specifiedColor=<rgba 0 0 0 0>}
      cornerRadii=<0 0 0 0>
     }
     tintColor {colorUseRule=UseColor_Specified, specifiedColor=<rgba 1 1 1 1>}
     uVRegion {bIsValid=false, max=<0 0>, min=<0 0>}
    }}
   slot {class=ButtonSlot, properties {padding=<ltrb 0 0 0 0>}}
  }
  Border_88 {
   widgetClassPath=&Border
   parent=Footer
   widget=Border_88
   properties {
    background {
     bIsDynamicallyLoaded=false
     drawAs=RoundedBox
     imageType=NoImage
     mirroring=NoMirror
     resourceName=None
     resourceObject=null
     tiling=NoTile
     imageSize=<32 32>
     margin=<ltrb 0 0 0 0>
     outlineSettings {
      bUseBrushTransparency=false
      roundingType=FixedRadius
      width=1
      color {colorUseRule=UseColor_Specified, specifiedColor=<rgba 0.1667 0.1667 0.1667 1>}
      cornerRadii=<4 4 4 4>
     }
     tintColor {colorUseRule=UseColor_Specified, specifiedColor=<rgba 1 1 1 0>}
     uVRegion {bIsValid=false, max=<0 0>, min=<0 0>}
    }
    padding=<ltrb 0 0 0 0>
   }
   slot {class=WrapBoxSlot, properties {bFillEmptySpace=true, fillSpanWhenLessThan=256, horizontalAlignment=HAlign_Right}}
  }
  SizeBox_143 {
   widgetClassPath=&SizeBox
   parent=Border_88
   widget=SizeBox_143
   properties {
    bOverride_MaxDesiredHeight=true
    bOverride_MaxDesiredWidth=true
    bOverride_MinDesiredHeight=true
    bOverride_MinDesiredWidth=true
    maxDesiredHeight=24
    maxDesiredWidth=400
    minDesiredHeight=24
    minDesiredWidth=200
   }
   slot {class=BorderSlot, properties {padding=<ltrb 0 0 0 0>}}
  }
  sld_ThumbSizer {
   widgetClassPath=/Script/Blutility.EditorUtilitySlider
   bIsVariable=true
   parent=SizeBox_143
   widget=sld_ThumbSizer
   properties {
    stepSize=0.1
    toolTipText=Controls the size of the assets thumbnails
    value=100
    widgetStyle {
     barThickness=4
     disabledBarImage {
      bIsDynamicallyLoaded=false
      drawAs=RoundedBox
      imageType=NoImage
      mirroring=NoMirror
      resourceName=None
      resourceObject=null
      tiling=NoTile
      imageSize=<200 0>
      margin=<ltrb 0 0 0 0>
      outlineSettings {
       bUseBrushTransparency=false
       roundingType=FixedRadius
       width=1
       color {colorUseRule=UseColor_ColorTable, specifiedColor=<rgba 0 0 0 0>}
       cornerRadii=<2 2 2 2>
      }
      tintColor {colorUseRule=UseColor_Specified, specifiedColor=<rgba 0.0048 0.0048 0.0048 1>}
      uVRegion {bIsValid=false, max=<0 0>, min=<0 0>}
     }
     disabledThumbImage {
      bIsDynamicallyLoaded=false
      drawAs=RoundedBox
      imageType=NoImage
      mirroring=NoMirror
      resourceName=None
      resourceObject=null
      tiling=NoTile
      imageSize=<8 8>
      margin=<ltrb 0 0 0 0>
      outlineSettings {
       bUseBrushTransparency=false
       roundingType=HalfHeightRadius
       width=0
       color {colorUseRule=UseColor_Specified, specifiedColor=<rgba 0 0 0 0>}
       cornerRadii=<0 0 0 0>
      }
      tintColor {colorUseRule=UseColor_Specified, specifiedColor=<rgba 0.5271 0.5271 0.5271 1>}
      uVRegion {bIsValid=false, max=<0 0>, min=<0 0>}
     }
     hoveredBarImage {
      bIsDynamicallyLoaded=false
      drawAs=RoundedBox
      imageType=NoImage
      mirroring=NoMirror
      resourceName=None
      resourceObject=null
      tiling=NoTile
      imageSize=<200 0>
      margin=<ltrb 0 0 0 0>
      outlineSettings {
       bUseBrushTransparency=false
       roundingType=FixedRadius
       width=1
       color {colorUseRule=UseColor_ColorTable, specifiedColor=<rgba 0 0 0 0>}
       cornerRadii=<2 2 2 2>
      }
      tintColor {colorUseRule=UseColor_Specified, specifiedColor=<rgba 0.0048 0.0048 0.0048 1>}
      uVRegion {bIsValid=false, max=<0 0>, min=<0 0>}
     }
     hoveredThumbImage {
      bIsDynamicallyLoaded=false
      drawAs=RoundedBox
      imageType=NoImage
      mirroring=NoMirror
      resourceName=None
      resourceObject=null
      tiling=NoTile
      imageSize=<8 8>
      margin=<ltrb 0 0 0 0>
      outlineSettings {
       bUseBrushTransparency=false
       roundingType=HalfHeightRadius
       width=0
       color {colorUseRule=UseColor_Specified, specifiedColor=<rgba 0 0 0 0>}
       cornerRadii=<0 0 0 0>
      }
      tintColor {colorUseRule=UseColor_Specified, specifiedColor=<rgba 1 1 1 1>}
      uVRegion {bIsValid=false, max=<0 0>, min=<0 0>}
     }
     normalBarImage {
      bIsDynamicallyLoaded=false
      drawAs=RoundedBox
      imageType=NoImage
      mirroring=NoMirror
      resourceName=None
      resourceObject=null
      tiling=NoTile
      imageSize=<200 0>
      margin=<ltrb 0 0 0 0>
      outlineSettings {
       bUseBrushTransparency=false
       roundingType=FixedRadius
       width=1
       color {colorUseRule=UseColor_ColorTable, specifiedColor=<rgba 0 0 0 0>}
       cornerRadii=<2 2 2 2>
      }
      tintColor {colorUseRule=UseColor_Specified, specifiedColor=<rgba 0.0048 0.0048 0.0048 1>}
      uVRegion {bIsValid=false, max=<0 0>, min=<0 0>}
     }
     normalThumbImage {
      bIsDynamicallyLoaded=false
      drawAs=RoundedBox
      imageType=NoImage
      mirroring=NoMirror
      resourceName=None
      resourceObject=null
      tiling=NoTile
      imageSize=<8 8>
      margin=<ltrb 0 0 0 0>
      outlineSettings {
       bUseBrushTransparency=false
       roundingType=HalfHeightRadius
       width=0
       color {colorUseRule=UseColor_Specified, specifiedColor=<rgba 0 0 0 0>}
       cornerRadii=<0 0 0 0>
      }
      tintColor {colorUseRule=UseColor_Specified, specifiedColor=<rgba 0.5271 0.5271 0.5271 1>}
      uVRegion {bIsValid=false, max=<0 0>, min=<0 0>}
     }
    }
   }
   slot {class=SizeBoxSlot, properties {padding=<ltrb 6 6 6 6>}}
  }
  MenuAnchorDialog {
   widgetClassPath=/Script/UMG.MenuAnchor
   bIsVariable=true
   parent=CanvasPanel_0
   widget=MenuAnchorDialog
   properties {
    menuClass=&EUWModalBase_C
    placement=MenuPlacement_Center
    useApplicationMenuStack=false
    visibility=SelfHitTestInvisible
   }
   slot {class=CanvasPanelSlot, properties {zOrder=10, layoutData {alignment=<0 0>, anchors {maximum=<1 1>, minimum=<0 0>}, offsets=<ltrb 0 0 0 0>}}}
  }
  dlgSlot {
   widgetClassPath=&NamedSlot
   bIsVariable=true
   parent=CanvasPanel_0
   widget=dlgSlot
   properties {visibility=Collapsed}
   slot {
    class=CanvasPanelSlot
    properties {zOrder=20, layoutData {alignment=<0 0>, anchors {maximum=<1 1>, minimum=<0 0>}, offsets=<ltrb 100 100 100 100>}}
   }
  }
 ]
}
