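# Editor utility list-entry tile for one Animation Library asset, showing a scrubbable thumbnail, name, type-colored outline and blinking dirty flag.
# Setting the list item stores the AnimLibAsset, fills the visuals and binds its dirty and writeable events.
# A right-click menu, rebuilt on writeable changes, offers apply, select controls, save, rename, overwrite, delete, thumbnail update and details, gating edits on writeability.
# Writeable tiles support drag-and-drop; ResizeWidget adapts thumbnail and text to scale.
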EditorUtilityWidgetBlueprint /AnimationLibrary/BP/EWC_AnimEntry
animations [DirtyOpacity {
  binding_count=1
  length_seconds=1
  object=&self:DirtyOpacity
  sequence {
   bindings [DirtyFlag {
     id=#1
     parent_id=null
     possessed_class=CommonTextBlock
     tracks [{
       type=MovieSceneFloatTrack
       display_name=Render Opacity
       sections [{type=MovieSceneFloatSection, end_frame=60, start_frame=0, channels [None {type=MovieSceneScriptingFloatChannel, keys [[0,1],[30,0],[60,1]]}]}]
      }]
    }]
   display_rate {denominator=1, numerator=60}
   tick_resolution {denominator=1, numerator=60000}
   tracks []
  }
 }]
graphs [
 EventGraph {nodes [
   n1 = "On Text Committed (renamer)"!() -> (OutputDelegate, then, Text:"Text (by ref)", CommitMethod:ETextCommit Enum=Default) {is=title K2Node_ComponentBoundEvent}
   > n2 = Switch on ETextCommit!(execute, Selection:ETextCommit Enum=@n1.CommitMethod) -> (Default:Exec, OnEnter:Exec, OnUserMovedFocus:Exec, OnCleared:Exec) {is=title K2Node_SwitchEnum}
   edit_name!(
    execute=[@n2.Default, @n2.OnEnter, @n2.OnUserMovedFocus]
    self:Anim Lib Asset Object Reference=Get AssetHandle() -> (AssetHandle:Anim Lib Asset Object Reference)
    NewName:String="To String (Text)"(InText=@n1.Text)
   ) -> (then)
   > n3 = Set Text(Text=@n1.Text, self=Get AssetName() -> (AssetName:Common Text Object Reference))
   Set Visibility(execute=[@n3.then, @n2.OnCleared], Visibility=Hidden, self=Get renamer() -> (renamer:Editor Utility Editable Text Box Object Reference))
   > Set isRenaming(isRenaming:bool=false) -> (Output_Get:bool)
   > Set Visibility(Visibility=Visible, self=Get BorderName() -> (BorderName:Border Object Reference))
   n4 = SaveAsset!() -> (OutputDelegate, then, Context Item:Editor Utility Widget Object Reference, Slot Widget:Editor Utility Widget Object Reference) {is=title K2Node_CustomEvent}
   n5 = Get ContextMenuAnchor() -> (ContextMenuAnchor:Menu Anchor Object Reference)
   > Close(self:Menu Anchor Object Reference=@n5)
   > save_asset(self:Anim Lib Asset Object Reference=Get AssetHandle())
   n6 = DeleteAsset!() -> (OutputDelegate, then, Context Item:Editor Utility Widget Object Reference, Slot Widget:Editor Utility Widget Object Reference) {is=title K2Node_CustomEvent}
   > Close(self=@n5)
   > OnDeleteAsset()
   n7 = OnWriteableChanged!() -> (OutputDelegate, then, IsWriteable:bool=~) {is=title K2Node_CustomEvent}
   n8 = Sequence!(execute=[@n45.then, @n7.then]) -> (then_0, then_1, then_2:Exec, then_3:Exec, then_4:Exec, then_5:Exec, then_6:Exec, then_7:Exec, then_8:Exec, then_9:Exec)
   n9 = Get ContextMenu() -> (ContextMenu:EUW Context Menu Object Reference)
   > Clear Context Menu(self:EUW Context Menu Object Reference=@n9)
   n10 = onApply!() -> (OutputDelegate, then, Context Item:Editor Utility Widget Object Reference, Slot Widget:Editor Utility Widget Object Reference) {is=title K2Node_CustomEvent}
   Add!(
    execute=@n8.then_1
    self:EUW Context Menu Object Reference=@n9
    Event:"Delegate (by ref)"=@n10
    Category:String=Anim
    Name:String=Apply
    Icon:Texture 2D Object Reference
    ItemWidget:User Widget Class Reference=/Game/NeverCook/UI/ContextMenu/EUW_ContextMenu_Button.EUW_ContextMenu_Button_C
    Tooltip:String=Apply Current Asset
   ) -> (then, Context Menu Item:Widget Object Reference, ContainerItem:User Widget Object Reference) {is=title K2Node_CallFunction}
   n11 = onSelectAll!() -> (OutputDelegate, then, Context Item:Editor Utility Widget Object Reference, Slot Widget:Editor Utility Widget Object Reference) {is=title K2Node_CustomEvent}
   Add(
    execute=@n8.then_2
    self=@n9
    Event=@n11
    Category=Controls
    Name=Select all
    ItemWidget=&EUW_ContextMenu_Button_C
    Tooltip=Select all the assets controls
   )
   n12 = Add(
    execute=@n8.then_3
    self=@n9
    Event=@n4
    Category=Edit
    Name=Save
    Icon=/AnimationLibrary/resources/icons/floppydisksolid
    ItemWidget=&EUW_ContextMenu_Button_C
    Tooltip=Save asset to disk and commit latest changes
   )
   > n13 = Set bIsEnabled!(
    execute
    bIsEnabled:bool=Get is_writeable!(self:Anim Lib Asset Object Reference=Get AssetHandle()) -> (is_writeable:bool)
    self:Widget Object Reference=@n12
   ) -> (then, Output_Get:bool)
   > Set bIsEnabled(bIsEnabled=@n13, self=@n12.ContainerItem)
   n14 = RenameItem!() -> (OutputDelegate, then, Context Item:Editor Utility Widget Object Reference, Slot Widget:Editor Utility Widget Object Reference) {is=title K2Node_CustomEvent}
   n15 = Add(
    execute=@n8.then_4
    self=@n9
    Event=@n14
    Category=Edit
    Name=Rename
    Icon=/AnimationLibrary/resources/icons/pentosquaresolid
    ItemWidget=&EUW_ContextMenu_Button_C
   )
   > n16 = Set bIsEnabled(bIsEnabled=Get is_writeable(self=Get AssetHandle()), self=@n15)
   > Set bIsEnabled(bIsEnabled=@n16, self=@n15.ContainerItem)
   n17 = onAssetOverwrite!() -> (OutputDelegate, then, Context Item:Editor Utility Widget Object Reference, Slot Widget:Editor Utility Widget Object Reference) {is=title K2Node_CustomEvent}
   n18 = Add(
    execute=@n8.then_5
    self=@n9
    Event=@n17
    Category=Edit
    Name=Overwrite
    ItemWidget=&EUW_ContextMenu_Button_C
    Tooltip=Copy data on this asset from sequencer
   )
   > n19 = Set bIsEnabled(bIsEnabled=Get is_writeable(self=Get AssetHandle()), self=@n18)
   > Set bIsEnabled(bIsEnabled=@n19, self=@n18.ContainerItem)
   n20 = Add(
    execute=@n8.then_6
    self=@n9
    Event=@n6
    Category=Edit
    Name=Delete
    Icon=/AnimationLibrary/resources/icons/trashsolid
    ItemWidget=&EUW_ContextMenu_Button_C
   )
   > n21 = Set bIsEnabled(bIsEnabled=Get is_writeable(self=Get AssetHandle()), self=@n20)
   > Set bIsEnabled(bIsEnabled=@n21, self=@n20.ContainerItem)
   n22 = CaptureThumbnail!() -> (OutputDelegate, then, Context Item:Editor Utility Widget Object Reference, Slot Widget:Editor Utility Widget Object Reference) {is=title K2Node_CustomEvent}
   n23 = Add(
    execute=@n8.then_7
    self=@n9
    Event=@n22
    Category=Thumbnail
    Name=Update Thumbnail
    ItemWidget=&EUW_ContextMenu_Button_C
   )
   > n24 = Set bIsEnabled(bIsEnabled=Get is_writeable(self=Get AssetHandle()), self=@n23)
   > n25 = Set bIsEnabled(bIsEnabled=@n24, self=@n23.ContainerItem)
   n26 = updatedThumbnail!() -> (OutputDelegate, then, ThumbPath:String, Task:Automation Editor Task Object Reference) {is=title K2Node_CustomEvent}
   n27 = Get AssetHandle()
   Bind Event to Thumbnail Updated(execute=@n25.then, self:Anim Lib Asset Object Reference=@n27, Delegate=@n26) {is=title K2Node_AddDelegate}
   n28 = onShowDetails!() -> (OutputDelegate, then, Context Item:Editor Utility Widget Object Reference, Slot Widget:Editor Utility Widget Object Reference) {is=title K2Node_CustomEvent}
   Add(
    execute=@n8.then_8
    self=@n9
    Event=@n28
    Category=Edit
    Name=View Details
    Icon=&pentosquaresolid
    ItemWidget=&EUW_ContextMenu_Button_C
    Tooltip=Show dialog with detailed information on the selected asset
   )
   n29 = OnIsDirtyChanged!() -> (OutputDelegate, then, isDirty:bool=~) {is=title K2Node_CustomEvent}
   n30 = Get DirtyFlag() -> (DirtyFlag:Common Text Object Reference)
   > n31 = Set Visibility(Visibility=Select(Option 0=Collapsed, Option 1=Visible, Index=@n29.isDirty), self=@n30)
   FlushAnimations(execute=[@n46.then, @n31.then], self:User Widget Object Reference)
   n32 = Get AssetHandle()
   n33 = Get dirty!(self:Anim Lib Asset Object Reference=@n32) -> (dirty:bool)
   > n34 = Branch(Condition=@n33)
   n35 = Get DirtyOpacity() -> (DirtyOpacity:Widget Animation Object Reference)
   > PlayAnimation(InAnimation=@n35, NumLoopsToPlay=0)
   StopAnimation(execute=@n34.else, InAnimation=@n35)
   n36 = Event PreConstruct()
   n37 = Sequence!(execute=[@n36.then, @n39.then]) -> (then_0, then_1, then_2, then_3)
   > Set NameSlateStyle(NameSlateStyle:Slate Color Structure=Get ColorAndOpacity!(self:Text Object Reference=Get AssetName()) -> (ColorAndOpacity:Slate Color Structure)) -> (Output_Get:Slate Color Structure)
   Set TypeSlateStyle(
    execute=@n37.then_1
    TypeSlateStyle:Slate Color Structure=Get ColorAndOpacity(self=Get AssetType() -> (AssetType:Common Text Object Reference))
   ) -> (Output_Get:Slate Color Structure)
   SetDesiredSizeOverride!(
    execute=@n37.then_2
    self:Image Object Reference=Get image!(self:Image Scrub Object Reference=Get ThumbImg() -> (ThumbImg:Image Scrub Object Reference)) -> (image:Image Object Reference)
    DesiredSize:Vector 2D Structure=(X=256,Y=256)
   ) -> (then)
   n38 = Create Widget!(
    execute:Exec=@n37.then_3
    WorldContextObject:Object Reference
    Class:User Widget Class Reference=/Game/NeverCook/UI/ContextMenu/EUW_ContextMenu.EUW_ContextMenu_C
    OwningPlayer:Player Controller Object Reference
   ) -> (then:Exec, ReturnValue:EUW Context Menu Object Reference)
   > Set ContextMenu(ContextMenu:EUW Context Menu Object Reference=@n38) -> (Output_Get:EUW Context Menu Object Reference)
   Event On Entry Released() {is=title K2Node_Event}
   > OnSelectedDisplay!(execute, self, selectionState:bool=false) -> (then)
   > UpdateResource(self:Image Scrub Object Reference=Get ThumbImg())
   n39 = Event OnInitialized()
   n40 = Event OnListItemObjectSet()
   > OnSelectedDisplay(selectionState=false)
   > n41 = Sequence!(execute) -> (then_0, then_1, then_2)
   > n42 = Cast To AnimLibAsset!(execute, Object=@n40.ListItemObject) -> (then, CastFailed, AsAnim Lib Asset:Anim Lib Asset Object Reference) {is=title K2Node_DynamicCast}
   > n43 = Set AssetHandle(AssetHandle:Anim Lib Asset Object Reference=@n42) -> (Output_Get:Anim Lib Asset Object Reference)
   > SetName!(execute, self, Name:String=Get Name!(self:Anim Lib Asset Object Reference=@n43) -> (Name:String)) -> (then)
   > SetBorderType!(execute, self, Type:String=Get Type!(self:Anim Lib Asset Object Reference=@n43) -> (Type:String)) -> (then)
   n44 = Get ToolTip!(self:Anim Lib Asset Object Reference=@n43) -> (ToolTip:String)
   > Set ToolTipText!(
    execute
    ToolTipText:Text="To Text (String)"(InString=Replace(SourceString=@n44, From="* ", To="\r\n"))
    self:Widget Object Reference=Get AssetTile() -> (AssetTile:Overlay Object Reference)
   ) -> (then, Output_Get:Text)
   > n45 = OnSetSource!(
    execute
    self:Image Scrub Object Reference=Get ThumbImg()
    mediasource:Media Source Object Reference=Get ThumbnailMediaSource!(self:Anim Lib Asset Object Reference=@n43) -> (ThumbnailMediaSource:Img Media Source Object Reference)
   ) -> (then)
   Bind Event to On Dirty Changed(execute=@n41.then_1, self:Anim Lib Asset Object Reference=@n32, Delegate=@n29) {is=title K2Node_AddDelegate}
   > n46 = Set Visibility(Visibility=Select(Option 0=Collapsed, Option 1=Visible, Index=@n33), self=@n30)
   Bind Event to On Writeable Changed(execute=@n41.then_2, self:Anim Lib Asset Object Reference=Get AssetHandle(), Delegate=@n7) {is=title K2Node_AddDelegate}
   n47 = Event On Item Selection Changed!() -> (OutputDelegate, then, bIsSelected:bool) {is=title K2Node_Event}
   > OnSelectedDisplay(selectionState=@n47.bIsSelected)
   Event OnMouseEnter()
   > n48 = Get!(execute:Exec) -> (then:Exec, else:Exec, isRenaming:bool)
   n49 = Branch(execute=@n48.else, Condition=IsVisible(self=Get ThumbImg()))
   n50 = Get BorderName()
   > n51 = Set Visibility(Visibility=Hidden, self=@n50)
   n52 = Set BorderOutlineWidth(execute=[@n49.else, @n51.then], BorderOutlineWidth:Vector="3,1,0") -> (Output_Get:Vector)
   SetBorderType(execute=[@n54.then, @n52.then], Type=Get Type(self=Get AssetHandle()))
   Event OnMouseLeave!() -> (OutputDelegate, then, MouseEvent:"Pointer Event Structure (by ref)") {is=title K2Node_Event}
   > n53 = Get()
   Set Visibility(execute=@n53.else, Visibility=Visible, self=@n50)
   > n54 = Set BorderOutlineWidth(BorderOutlineWidth="2,0.5,0")
   update_thumbnail!(execute=@n63.then, self:Anim Lib Asset Object Reference=@n27) -> (then, SequencePath:String)
   Rename(execute=@n61.then)
   SelectControlsInrig!(execute=@n64.then, self:Anim Lib Asset Object Reference=Get AssetHandle()) -> (then, Control Rig:Control Rig Object Reference)
   OnOpenDetails(execute=@n59.then)
   OnOverwriteAsset(execute=@n62.then)
   n55 = GetLibraryParent!(execute=@n60.then, self) -> (then, LibraryEUW:EUW Anim Library Object Reference)
   > OnAssetContextApply(self:EUW Anim Library Object Reference=@n55)
   n56 = Get ThumbImg()
   OnSetSource(execute=@n58.then, self=@n56, mediasource=Get ThumbnailMediaSource(self=@n27))
   Reroute Node(InputPin:Wildcard) -> (OutputPin:Wildcard)
   n57 = GetEditorWorld(execute=@n26.then, self=Get UnrealEditorSubsystem())
   > UpdateResource(self=@n56)
   > n58 = DelayUntilNextTick!(execute, WorldContextObject:Object Reference=@n57) -> (then)
   n59 = Close(execute=@n28.then, self=@n5)
   n60 = Close(execute=@n10.then, self=@n5)
   n61 = Close(execute=@n14.then, self=@n5)
   n62 = Close(execute=@n17.then, self=@n5)
   n63 = Close(execute=@n22.then, self=@n5)
   n64 = Close(execute=@n11.then, self=@n5)
  ]}
 GetLibraryParent {nodes [
   GetLibraryParent() {is=title K2Node_FunctionEntry}
   > n1 = FindParentWidgetOfType!(
    execute
    StartingWidget:Widget Object Reference=Self-Reference()
    Type:Widget Class Reference=/AnimationLibrary/EUW_AnimLibrary.EUW_AnimLibrary_C
   ) -> (then, ReturnValue:EUW Anim Library Object Reference)
   > Return Node!(execute, LibraryEUW:EUW Anim Library Object Reference=@n1)
  ]}
 GetVisibility_0 {nodes [
   GetVisibility_0() {is=title K2Node_FunctionEntry}
   n1 = Get AssetHandle()
   > Return Node!(execute, ReturnValue:ESlateVisibility Enum=Select(Option 0=Visible, Option 1=Hidden, Index=Get is_writeable(self=@n1)))
  ]}
 IsSelectedNull {nodes [
   n1 = IsSelectedNull!() -> (then, ItemObject:Object Reference) {is=title K2Node_FunctionEntry}
   > Return Node!(execute, Is Null Item:bool=NOT Boolean(A=IsValid(Object=@n1)))
  ]}
 OnDeleteAsset {nodes [
   OnDeleteAsset() {is=title K2Node_FunctionEntry}
   > n1 = Show Message Dialog(
    Title=MakeLiteralText(Value="LOCTEXT(\"Animation Library\")")
    Message=MakeLiteralText(Value="LOCTEXT(\"You\\'re about to delete this asset.\\r\\nAre your sure?\")")
    MessageType=YesNo
    DefaultValue=Yes
    MessageCategory=Warning
   )
   > n2 = Switch on EAppReturnType(Selection=@n1)
   delete(execute=@n2.Yes, self:Anim Lib Asset Object Reference=Get AssetHandle())
  ]}
 OnDragDetected {nodes [
   OnDragDetected!() -> (then, MyGeometry:Geometry Structure, PointerEvent:"Pointer Event Structure (by ref)") {is=title K2Node_FunctionEntry}
   > Branch(Condition=Get is_writeable(self=Get AssetHandle()))
   > PrintString(InString=Dragging Asset)
   > n1 = Create Widget(Class=/Game/NeverCook/ArtTools/Tech_Art/AssetPlaylist/EUW_AssetPlaylist_DragDropWidget.EUW_AssetPlaylist_DragDropWidget_C) -> (ReturnValue:EUW Asset Playlist Drag Drop Widget Object Reference)
   > SetBrushFromTexture!(
    execute
    self:Image Object Reference=Get Image_0!(self:EUW Asset Playlist Drag Drop Widget Object Reference=@n1) -> (Image_0:Image Object Reference)
    Texture:Texture 2D Object Reference=/AnimationLibrary/resources/icons/personwalkingsolid
    bMatchSize:bool=false
   ) -> (then)
   > n2 = "Create Drag && Drop Operation"!(
    execute:Exec
    WorldContextObject:Object Reference
    Class:Drag Drop Operation Class Reference=/Script/UMG.DragDropOperation
    Tag:String
    Payload:Object Reference=Get AssetHandle()
    DefaultDragVisual:Widget Object Reference=@n1
    Pivot:EDragPivot Enum=CenterCenter
    Offset:Vector 2D Structure=(X=0,Y=0)
   ) -> (then:Exec, ReturnValue:Drag Drop Operation Object Reference) {is=title K2Node_CreateDragDropOperation}
   > Return Node!(execute, Operation:Drag Drop Operation Object Reference=@n2)
  ]}
 OnMouseButtonDown {nodes [
   n1 = OnMouseButtonDown!() -> (then, MyGeometry:Geometry Structure, MouseEvent:"Pointer Event Structure (by ref)") {is=title K2Node_FunctionEntry}
   > n2 = Branch(Condition=Is Mouse Button Down!(Input:"Pointer Event Structure (by ref)"=@n1.MouseEvent, MouseButton:Key Structure=RightMouseButton) -> (ReturnValue:bool))
   > ToggleOpen!(execute, self:Menu Anchor Object Reference=Get ContextMenuAnchor(), bFocusOnOpen:bool=true) -> (then)
   n3 = Unhandled()
   > Return Node!(execute, ReturnValue:Event Reply Structure=@n3)
   Return Node(execute=@n2.else, ReturnValue=@n3)
   Handled()
  ]}
 OnOpenDetails {nodes [
   OnOpenDetails() {is=title K2Node_FunctionEntry}
   > n1 = GetLibraryParent() {is=title K2Node_CallFunction}
   > n2 = Construct Object from Class(Class=&/EUW_AnimLibrary_Tab_Actions.EUW_AnimLibrary_Tab_Actions_C) -> (ReturnValue:EUW Anim Library Tab Actions Object Reference)
   > Set parent!(execute, parent:EUW Anim Library Object Reference=@n1, self:EUW Anim Library Tab Actions Object Reference=@n2) -> (then, Output_Get:EUW Anim Library Object Reference)
   > Set current_asset!(execute, current_asset:Anim Lib Asset Object Reference=Get AssetHandle(), self:EUW Anim Library Tab Actions Object Reference=@n2) -> (then, Output_Get:Anim Lib Asset Object Reference)
   n3 = Get Name(self=Get AssetHandle())
   > ShowEditorUtilityDialogInstance!(
    execute
    WidgetInstance:Editor Utility Dialog Widget Object Reference=@n2
    OnClosed:"Delegate (by ref)"
    Title:Text="To Text (String)"(InString=Append(A="Details: ", B=@n3))
    ButtonLayout:EAppMsgType Enum=OkCancel
    Size:Vector 2D Structure=(X=320,Y=700)
   ) -> (then)
   ShowModalEditorUtilityDialogInstance!(
    execute
    WidgetInstance:Editor Utility Dialog Widget Object Reference
    Title:Text
    ButtonLayout:EAppMsgType Enum=OkCancel
    Size:Vector 2D Structure=(X=240,Y=480)
   ) -> (then, ReturnValue:Editor Utility Dialog Result Structure)
  ]}
 OnOverwriteAsset {nodes [
   OnOverwriteAsset() {is=title K2Node_FunctionEntry}
   > n1 = GetLibraryParent()
   n2 = Get AssetHandle()
   > MakeAsset!(
    execute
    self:EUW Anim Library Object Reference=@n1
    AsPose:bool="Equal Exactly (String)"(A=Get Type(self=@n2), B=Pose)
    AsNew:bool=false
    TempAsset:Anim Lib Asset Object Reference
   ) -> (then, Asset:Anim Lib Asset Object Reference)
  ]}
 OnSelectedDisplay {nodes [
   n1 = OnSelectedDisplay!() -> (then, selectionState:bool=~) {is=title K2Node_FunctionEntry}
   > SetBrushColor(
    self=Get asset_border() -> (asset_border:Border Object Reference)
    InBrushColor=SelectColor(A=(R=0,G=0.1608,B=0.7451,A=1), B=(R=0,G=0,B=0,A=0), bPickA=@n1)
   )
   n2 = Make Slate Color(SpecifiedColor=(R=1,G=1,B=1,A=1))
   > SetColorAndOpacity(
    self=Get AssetName()
    InColorAndOpacity=Select(Option 0:Slate Color Structure=Get NameSlateStyle() -> (NameSlateStyle:Slate Color Structure), Option 1:Slate Color Structure=@n2, Index=@n1) -> (ReturnValue:Slate Color Structure)
   )
   > SetColorAndOpacity(
    self=Get AssetType()
    InColorAndOpacity=Select(Option 0=Get TypeSlateStyle() -> (TypeSlateStyle:Slate Color Structure), Option 1=@n2, Index=@n1)
   )
  ]}
 On_ContextMenuAnchor_GetUserMenuContent {nodes [
   On_ContextMenuAnchor_GetUserMenuContent() {is=title K2Node_FunctionEntry}
   n1 = Get ContextMenu()
   > BuildContent(self:EUW Context Menu Object Reference=@n1)
   > Return Node(ReturnValue:User Widget Object Reference=@n1)
  ]}
 Rename {nodes [
   Rename() {is=title K2Node_FunctionEntry}
   n1 = Get AssetHandle()
   n2 = Get renamer()
   > Set Text(Text="To Text (String)"(InString=Get Name(self=@n1)), self:Text Box Object Reference=@n2)
   > Set Visibility(Visibility=Visible, self=@n2)
   > SetFocus(self=@n2)
   > SetKeyboardFocus(self:Widget Object Reference=@n2)
   > Set Visibility(Visibility=Hidden, self=Get BorderName())
   > Set isRenaming(isRenaming=true)
  ]}
 Resize {nodes [Resize!() -> (then, Size:double=~) {is=title K2Node_FunctionEntry}]}
 ResizeWidget {nodes [
   n1 = ResizeWidget!() -> (then, Scale:double=~) {is=title K2Node_FunctionEntry}
   n2 = NOT Boolean(A="float < float"(A=@n1, B=0.25))
   > Set Visibility(
    Visibility=Select(Option 0:ESlateVisibility Enum=Collapsed, Option 1:ESlateVisibility Enum=Visible, Index=@n2) -> (ReturnValue:ESlateVisibility Enum)
    self=Get ThumbImg()
   )
   n3 = Get DefaultFontSize() -> (DefaultFontSize:double)
   n4 = Get AssetName()
   > n5 = Set members in Slate Font Info!(
    execute:Exec
    StructRef:"Slate Font Info Structure (by ref)"=Get Font!(self:Text Object Reference=@n4) -> (Font:Slate Font Info Structure)
    Size:float="Clamp (Float)"(Value=float * float(A=@n3, B=@n1) {is=title K2Node_PromotableOperator}, Min=8, Max=@n3)
   ) -> (then:Exec, StructOut:"Slate Font Info Structure (by ref)") {is=title K2Node_SetFieldsInStruct}
   > n6 = Set Font!(execute, Font:Slate Font Info Structure=@n5, self:Text Object Reference=@n4) -> (then, Output_Get:Slate Font Info Structure)
   > n7 = Set AutoWrapText!(execute, AutoWrapText:bool=@n2, self:Text Layout Widget Object Reference=@n4) -> (then, Output_Get:bool)
   > Branch(Condition=@n7)
   > Set WrapTextAt!(execute, WrapTextAt:float=float * float(A=@n1, B=240), self:Text Layout Widget Object Reference=@n4) -> (then, Output_Get:float)
   n8 = Get renamer()
   n9 = Get WidgetStyle!(self:Text Box Object Reference=@n8) -> (WidgetStyle:Editable Text Box Style Structure)
   n10 = Set members in Text Block Style!(
    execute:Exec
    StructRef:"Text Block Style Structure (by ref)"=Break Editable Text Box Style!(EditableTextBoxStyle:"Editable Text Box Style Structure (by ref)"=@n9) -> (TextStyle:Text Block Style Structure) {is=title K2Node_BreakStruct}
    Font:Slate Font Info Structure=@n6
   ) -> (then:Exec, StructOut:"Text Block Style Structure (by ref)") {is=title K2Node_SetFieldsInStruct}
   > n11 = Set members in Editable Text Box Style!(execute:Exec, StructRef:"Editable Text Box Style Structure (by ref)"=@n9, TextStyle:Text Block Style Structure=@n10) -> (then:Exec, StructOut:"Editable Text Box Style Structure (by ref)") {is=title K2Node_SetFieldsInStruct}
   > Set WidgetStyle!(execute, WidgetStyle:Editable Text Box Style Structure=@n11, self:Text Box Object Reference=@n8) -> (then, Output_Get:Editable Text Box Style Structure)
  ]}
 SetBorderOutline {nodes [
   n1 = SetBorderOutline!() -> (then, Outline:Slate Color Structure, Width:float=~) {is=title K2Node_FunctionEntry}
   n2 = Get asset_border()
   n3 = Break Slate Brush!(SlateBrush:"Slate Brush Structure (by ref)"=Get Background!(self:Border Object Reference=@n2) -> (Background:Slate Brush Structure)) -> (
    TintColor:Slate Color Structure
    DrawAs:ESlateBrushDrawType Enum
    Tiling:ESlateBrushTileType Enum
    Mirroring:ESlateBrushMirrorType Enum
    ImageSize:Vector 2D Structure
    Margin:Margin Structure
    ResourceObject:Object Reference
    OutlineSettings:Slate Brush Outline Settings Structure
   ) {is=title K2Node_BreakStruct}
   n4 = Break Slate Brush Outline Settings!(SlateBrushOutlineSettings:"Slate Brush Outline Settings Structure (by ref)"=@n3.OutlineSettings) -> (
    CornerRadii:Vector 4 Structure
    Color:Slate Color Structure
    Width:float=~
    RoundingType:ESlateBrushRoundingType Enum
    bUseBrushTransparency:bool=~
   ) {is=title K2Node_BreakStruct}
   > n5 = Set members in Slate Brush Outline Settings!(
    execute:Exec
    StructRef:"Slate Brush Outline Settings Structure (by ref)"=@n3.OutlineSettings
    CornerRadii:Vector 4 Structure=@n4
    Color:Slate Color Structure=@n1
    Width:float=@n1.Width
    RoundingType:ESlateBrushRoundingType Enum=@n4.RoundingType
    bUseBrushTransparency:bool=@n4.bUseBrushTransparency
   ) -> (then:Exec, StructOut:"Slate Brush Outline Settings Structure (by ref)") {is=title K2Node_SetFieldsInStruct}
   > SetBrush(
    self:Border Object Reference=@n2
    InBrush=Make Slate Brush!(
     TintColor:Slate Color Structure=@n3
     DrawAs:ESlateBrushDrawType Enum=@n3.DrawAs
     Tiling:ESlateBrushTileType Enum=@n3.Tiling
     Mirroring:ESlateBrushMirrorType Enum=@n3.Mirroring
     ImageSize:Vector 2D Structure=@n3.ImageSize
     Margin:Margin Structure=@n3.Margin
     ResourceObject:Object Reference=@n3.ResourceObject
     OutlineSettings:Slate Brush Outline Settings Structure=@n5
    ) -> (SlateBrush:Slate Brush Structure) {is=title K2Node_MakeStruct}
   )
   Get BorderOutlineWidth() -> (BorderOutlineWidth:Vector)
  ]}
 SetBorderType {nodes [
   n1 = SetBorderType!() -> (then, Type:String) {is=title K2Node_FunctionEntry}
   n2 = BreakVector(InVec=Get BorderOutlineWidth())
   > n3 = Set members in Linear Color!(
    execute:Exec
    StructRef:"Linear Color Structure (by ref)"=Select(
     Option 0:Linear Color Structure=(R=1,G=0.5,B=0,A=1)
     Option 1:Linear Color Structure=(R=0,G=1,B=1,A=1)
     Index="Equal Exactly (String)"(A=@n1, B=Animation)
    ) -> (ReturnValue:Linear Color Structure)
    A:float=@n2.Y
   ) -> (then:Exec, StructOut:"Linear Color Structure (by ref)") {is=title K2Node_SetFieldsInStruct}
   > SetBorderOutline!(execute, self, Outline:Slate Color Structure=Make Slate Color(SpecifiedColor=@n3), Width:float=@n2) -> (then) {is=title K2Node_CallFunction}
  ]}
 SetName {nodes [
   n1 = SetName!() -> (then, Name:String) {is=title K2Node_FunctionEntry}
   > Set Text(Text="To Text (String)"(InString=@n1), self:Text Object Reference=Get AssetName())
  ]}
]
variables [
 AssetHandle {type=AnimLibAsset_C, default=null}
 BorderOutlineWidth {type=Vector, default=<1 0.5 0>}
 ContextMenu {type=EUW_ContextMenu_C, default=null}
 DefaultFontSize {type=number, default=18}
 isRenaming {type=boolean}
 NameSlateStyle {type=SlateColor, default {colorUseRule=UseColor_Specified, specifiedColor=<rgba 1 0 1 1>}}
 Resize {default=()}
 TypeSlateStyle {type=SlateColor, default {colorUseRule=UseColor_Specified, specifiedColor=<rgba 1 0 1 1>}}
]
widgets {
 info {rootWidgetClass=&Border, widgetCount=10}
 widgets [
  asset_border {
   widgetClassPath=&Border
   bIsVariable=true
   widget=asset_border
   properties {
    background {
     bIsDynamicallyLoaded=false
     drawAs=RoundedBox
     imageType=NoImage
     mirroring=NoMirror
     resourceName=None
     resourceObject=null
     tiling=NoTile
     imageSize=<32 32>
     margin=<ltrb 0 0 0 0>
     outlineSettings {
      bUseBrushTransparency=false
      roundingType=FixedRadius
      width=2
      color {colorUseRule=UseColor_Specified, specifiedColor=<rgba 0 0 0 0>}
      cornerRadii=<8 8 8 8>
     }
     tintColor {colorUseRule=UseColor_Specified, specifiedColor=<rgba 1 1 1 1>}
     uVRegion {bIsValid=false, max=<0 0>, min=<0 0>}
    }
    brushColor=<rgba 1 1 1 0>
    padding=<ltrb 6 6 6 6>
   }
  }
  AssetTile {
   widgetClassPath=&Overlay
   bIsVariable=true
   parent=asset_border
   widget=AssetTile
   properties {bOverride_Cursor=true, visibility=Visible}
   slot {class=BorderSlot, properties {padding=<ltrb 6 6 6 6>}}
  }
  ThumbImg {
   widgetClassPath=&/ImageScrub
   bIsVariable=true
   parent=AssetTile
   widget=ThumbImg
   properties {bOverride_Cursor=true, visibility=Visible, foregroundColor {colorUseRule=UseColor_Foreground, specifiedColor=<rgba 1 1 1 1>}}
   slot {class=OverlaySlot, properties {horizontalAlignment=HAlign_Fill, verticalAlignment=VAlign_Fill}}
  }
  AssetType {
   widgetClassPath=&CommonTextBlock
   bIsVariable=true
   parent=AssetTile
   widget=AssetType
   properties {
    bIsEnabled=false
    visibility=Collapsed
    colorAndOpacity {colorUseRule=UseColor_Specified, specifiedColor=<rgba 1 1 1 0.6696>}
    font {
     bForceMonospaced=false
     bMaterialIsStencil=false
     fontMaterial=null
     fontObject=&Roboto
     letterSpacing=0
     monospacedWidth=1
     size=9
     skewAmount=0
     typefaceFontName=Regular
     outlineSettings {
      bApplyOutlineToDropShadows=false
      bMiteredCorners=false
      bSeparateFillAlpha=false
      outlineMaterial=null
      outlineSize=0
      outlineColor=<rgba 0 0 0 1>
     }
    }
    shadowOffset=<0 0>
   }
   slot {class=OverlaySlot}
  }
  renamer {
   widgetClassPath=&EditorUtilityEditableTextBox
   bIsVariable=true
   parent=AssetTile
   widget=renamer
   properties {
    hintText=ALICE
    visibility=Hidden
    widgetStyle {
     backgroundColor {colorUseRule=UseColor_Specified, specifiedColor=<rgba 1 1 1 1>}
     backgroundImageFocused {
      bIsDynamicallyLoaded=false
      drawAs=RoundedBox
      imageType=NoImage
      mirroring=NoMirror
      resourceName=None
      resourceObject=null
      tiling=NoTile
      imageSize=<0 0>
      margin=<ltrb 0 0 0 0>
      outlineSettings {
       bUseBrushTransparency=false
       roundingType=FixedRadius
       width=1
       color {colorUseRule=UseColor_ColorTable, specifiedColor=<rgba 0 0 0 0>}
       cornerRadii=<4 4 4 4>
      }
      tintColor {colorUseRule=UseColor_Specified, specifiedColor=<rgba 0.0048 0.0048 0.0048 1>}
      uVRegion {bIsValid=false, max=<0 0>, min=<0 0>}
     }
     backgroundImageHovered {
      bIsDynamicallyLoaded=false
      drawAs=RoundedBox
      imageType=NoImage
      mirroring=NoMirror
      resourceName=None
      resourceObject=null
      tiling=NoTile
      imageSize=<0 0>
      margin=<ltrb 0 0 0 0>
      outlineSettings {
       bUseBrushTransparency=false
       roundingType=FixedRadius
       width=1
       color {colorUseRule=UseColor_ColorTable, specifiedColor=<rgba 0 0 0 0>}
       cornerRadii=<4 4 4 4>
      }
      tintColor {colorUseRule=UseColor_Specified, specifiedColor=<rgba 0.0048 0.0048 0.0048 1>}
      uVRegion {bIsValid=false, max=<0 0>, min=<0 0>}
     }
     backgroundImageNormal {
      bIsDynamicallyLoaded=false
      drawAs=RoundedBox
      imageType=NoImage
      mirroring=NoMirror
      resourceName=None
      resourceObject=null
      tiling=NoTile
      imageSize=<0 0>
      margin=<ltrb 0 0 0 0>
      outlineSettings {
       bUseBrushTransparency=false
       roundingType=FixedRadius
       width=1
       color {colorUseRule=UseColor_ColorTable, specifiedColor=<rgba 0 0 0 0>}
       cornerRadii=<4 4 4 4>
      }
      tintColor {colorUseRule=UseColor_Specified, specifiedColor=<rgba 0.0209 0.0209 0.0209 1>}
      uVRegion {bIsValid=false, max=<0 0>, min=<0 0>}
     }
     backgroundImageReadOnly {
      bIsDynamicallyLoaded=false
      drawAs=RoundedBox
      imageType=NoImage
      mirroring=NoMirror
      resourceName=None
      resourceObject=null
      tiling=NoTile
      imageSize=<0 0>
      margin=<ltrb 0 0 0 0>
      outlineSettings {
       bUseBrushTransparency=false
       roundingType=FixedRadius
       width=1
       color {colorUseRule=UseColor_ColorTable, specifiedColor=<rgba 0 0 0 0>}
       cornerRadii=<4 4 4 4>
      }
      tintColor {colorUseRule=UseColor_Specified, specifiedColor=<rgba 0.0284 0.0284 0.0284 1>}
      uVRegion {bIsValid=false, max=<0 0>, min=<0 0>}
     }
     focusedForegroundColor {colorUseRule=UseColor_Specified, specifiedColor=<rgba 0.807 0.807 0.807 1>}
     foregroundColor {colorUseRule=UseColor_Specified, specifiedColor=<rgba 0.807 0.807 0.807 1>}
     hScrollBarPadding=<ltrb 0 0 0 0>
     padding=<ltrb 2 2 2 2>
     readOnlyForegroundColor {colorUseRule=UseColor_Foreground, specifiedColor=<rgba 1 0 1 1>}
     scrollBarStyle {
      thickness=8
      draggedThumbImage {
       bIsDynamicallyLoaded=false
       drawAs=RoundedBox
       imageType=NoImage
       mirroring=NoMirror
       resourceName=None
       resourceObject=null
       tiling=NoTile
       imageSize=<0 0>
       margin=<ltrb 0 0 0 0>
       outlineSettings {
        bUseBrushTransparency=false
        roundingType=FixedRadius
        width=0
        color {colorUseRule=UseColor_Specified, specifiedColor=<rgba 0 0 0 0>}
        cornerRadii=<4 4 4 4>
       }
       tintColor {colorUseRule=UseColor_Specified, specifiedColor=<rgba 0.2159 0.2159 0.2159 1>}
       uVRegion {bIsValid=false, max=<0 0>, min=<0 0>}
      }
      horizontalBackgroundImage {
       bIsDynamicallyLoaded=false
       drawAs=NoDrawType
       imageType=NoImage
       mirroring=NoMirror
       resourceName=None
       resourceObject=null
       tiling=NoTile
       imageSize=<0 0>
       margin=<ltrb 0 0 0 0>
       outlineSettings {
        bUseBrushTransparency=false
        roundingType=HalfHeightRadius
        width=0
        color {colorUseRule=UseColor_Specified, specifiedColor=<rgba 0 0 0 0>}
        cornerRadii=<0 0 0 0>
       }
       tintColor {colorUseRule=UseColor_Specified, specifiedColor=<rgba 1 1 1 1>}
       uVRegion {bIsValid=false, max=<0 0>, min=<0 0>}
      }
      horizontalBottomSlotImage {
       bIsDynamicallyLoaded=false
       drawAs=NoDrawType
       imageType=NoImage
       mirroring=NoMirror
       resourceName=None
       resourceObject=null
       tiling=NoTile
       imageSize=<0 0>
       margin=<ltrb 0 0 0 0>
       outlineSettings {
        bUseBrushTransparency=false
        roundingType=HalfHeightRadius
        width=0
        color {colorUseRule=UseColor_Specified, specifiedColor=<rgba 0 0 0 0>}
        cornerRadii=<0 0 0 0>
       }
       tintColor {colorUseRule=UseColor_Specified, specifiedColor=<rgba 1 1 1 1>}
       uVRegion {bIsValid=false, max=<0 0>, min=<0 0>}
      }
      horizontalTopSlotImage {
       bIsDynamicallyLoaded=false
       drawAs=NoDrawType
       imageType=NoImage
       mirroring=NoMirror
       resourceName=None
       resourceObject=null
       tiling=NoTile
       imageSize=<0 0>
       margin=<ltrb 0 0 0 0>
       outlineSettings {
        bUseBrushTransparency=false
        roundingType=HalfHeightRadius
        width=0
        color {colorUseRule=UseColor_Specified, specifiedColor=<rgba 0 0 0 0>}
        cornerRadii=<0 0 0 0>
       }
       tintColor {colorUseRule=UseColor_Specified, specifiedColor=<rgba 1 1 1 1>}
       uVRegion {bIsValid=false, max=<0 0>, min=<0 0>}
      }
      hoveredThumbImage {
       bIsDynamicallyLoaded=false
       drawAs=RoundedBox
       imageType=NoImage
       mirroring=NoMirror
       resourceName=None
       resourceObject=null
       tiling=NoTile
       imageSize=<0 0>
       margin=<ltrb 0 0 0 0>
       outlineSettings {
        bUseBrushTransparency=false
        roundingType=FixedRadius
        width=0
        color {colorUseRule=UseColor_Specified, specifiedColor=<rgba 0 0 0 0>}
        cornerRadii=<4 4 4 4>
       }
       tintColor {colorUseRule=UseColor_Specified, specifiedColor=<rgba 0.2159 0.2159 0.2159 1>}
       uVRegion {bIsValid=false, max=<0 0>, min=<0 0>}
      }
      normalThumbImage {
       bIsDynamicallyLoaded=false
       drawAs=RoundedBox
       imageType=NoImage
       mirroring=NoMirror
       resourceName=None
       resourceObject=null
       tiling=NoTile
       imageSize=<0 0>
       margin=<ltrb 0 0 0 0>
       outlineSettings {
        bUseBrushTransparency=false
        roundingType=FixedRadius
        width=0
        color {colorUseRule=UseColor_Specified, specifiedColor=<rgba 0 0 0 0>}
        cornerRadii=<4 4 4 4>
       }
       tintColor {colorUseRule=UseColor_Specified, specifiedColor=<rgba 0.0953 0.0953 0.0953 1>}
       uVRegion {bIsValid=false, max=<0 0>, min=<0 0>}
      }
      verticalBackgroundImage {
       bIsDynamicallyLoaded=false
       drawAs=NoDrawType
       imageType=NoImage
       mirroring=NoMirror
       resourceName=None
       resourceObject=null
       tiling=NoTile
       imageSize=<0 0>
       margin=<ltrb 0 0 0 0>
       outlineSettings {
        bUseBrushTransparency=false
        roundingType=HalfHeightRadius
        width=0
        color {colorUseRule=UseColor_Specified, specifiedColor=<rgba 0 0 0 0>}
        cornerRadii=<0 0 0 0>
       }
       tintColor {colorUseRule=UseColor_Specified, specifiedColor=<rgba 1 1 1 1>}
       uVRegion {bIsValid=false, max=<0 0>, min=<0 0>}
      }
      verticalBottomSlotImage {
       bIsDynamicallyLoaded=false
       drawAs=NoDrawType
       imageType=NoImage
       mirroring=NoMirror
       resourceName=None
       resourceObject=null
       tiling=NoTile
       imageSize=<0 0>
       margin=<ltrb 0 0 0 0>
       outlineSettings {
        bUseBrushTransparency=false
        roundingType=HalfHeightRadius
        width=0
        color {colorUseRule=UseColor_Specified, specifiedColor=<rgba 0 0 0 0>}
        cornerRadii=<0 0 0 0>
       }
       tintColor {colorUseRule=UseColor_Specified, specifiedColor=<rgba 1 1 1 1>}
       uVRegion {bIsValid=false, max=<0 0>, min=<0 0>}
      }
      verticalTopSlotImage {
       bIsDynamicallyLoaded=false
       drawAs=NoDrawType
       imageType=NoImage
       mirroring=NoMirror
       resourceName=None
       resourceObject=null
       tiling=NoTile
       imageSize=<0 0>
       margin=<ltrb 0 0 0 0>
       outlineSettings {
        bUseBrushTransparency=false
        roundingType=HalfHeightRadius
        width=0
        color {colorUseRule=UseColor_Specified, specifiedColor=<rgba 0 0 0 0>}
        cornerRadii=<0 0 0 0>
       }
       tintColor {colorUseRule=UseColor_Specified, specifiedColor=<rgba 1 1 1 1>}
       uVRegion {bIsValid=false, max=<0 0>, min=<0 0>}
      }
     }
     textStyle {
      overflowPolicy=Clip
      transformPolicy=None
      colorAndOpacity {colorUseRule=UseColor_Foreground, specifiedColor=<rgba 1 0 1 1>}
      font {
       bForceMonospaced=false
       bMaterialIsStencil=false
       fontMaterial=null
       fontObject=&Roboto
       letterSpacing=0
       monospacedWidth=1
       size=13.5
       skewAmount=0
       typefaceFontName=Regular
       outlineSettings {
        bApplyOutlineToDropShadows=false
        bMiteredCorners=false
        bSeparateFillAlpha=false
        outlineMaterial=null
        outlineSize=0
        outlineColor=<rgba 0 0 0 0.5>
       }
      }
      highlightColor {colorUseRule=UseColor_ColorTable, specifiedColor=<rgba 0 0 0 0>}
      highlightShape {
       bIsDynamicallyLoaded=false
       drawAs=Image
       imageType=NoImage
       mirroring=NoMirror
       resourceName=None
       resourceObject=null
       tiling=NoTile
       imageSize=<0 0>
       margin=<ltrb 0 0 0 0>
       outlineSettings {
        bUseBrushTransparency=false
        roundingType=HalfHeightRadius
        width=0
        color {colorUseRule=UseColor_Specified, specifiedColor=<rgba 0 0 0 0>}
        cornerRadii=<0 0 0 0>
       }
       tintColor {colorUseRule=UseColor_ColorTable, specifiedColor=<rgba 0 0 0 0>}
       uVRegion {bIsValid=false, max=<0 0>, min=<0 0>}
      }
      selectedBackgroundColor {colorUseRule=UseColor_ColorTable, specifiedColor=<rgba 0 0 0 0>}
      shadowColorAndOpacity=<rgba 0 0 0 1>
      shadowOffset=<0 0>
      strikeBrush {
       bIsDynamicallyLoaded=false
       drawAs=Image
       imageType=NoImage
       mirroring=NoMirror
       resourceName=None
       resourceObject=null
       tiling=NoTile
       imageSize=<32 32>
       margin=<ltrb 0 0 0 0>
       outlineSettings {
        bUseBrushTransparency=false
        roundingType=HalfHeightRadius
        width=0
        color {colorUseRule=UseColor_Specified, specifiedColor=<rgba 0 0 0 0>}
        cornerRadii=<0 0 0 0>
       }
       tintColor {colorUseRule=UseColor_Specified, specifiedColor=<rgba 1 1 1 1>}
       uVRegion {bIsValid=false, max=<0 0>, min=<0 0>}
      }
      underlineBrush {
       bIsDynamicallyLoaded=false
       drawAs=Image
       imageType=NoImage
       mirroring=NoMirror
       resourceName=None
       resourceObject=null
       tiling=NoTile
       imageSize=<32 32>
       margin=<ltrb 0 0 0 0>
       outlineSettings {
        bUseBrushTransparency=false
        roundingType=HalfHeightRadius
        width=0
        color {colorUseRule=UseColor_Specified, specifiedColor=<rgba 0 0 0 0>}
        cornerRadii=<0 0 0 0>
       }
       tintColor {colorUseRule=UseColor_Specified, specifiedColor=<rgba 1 1 1 1>}
       uVRegion {bIsValid=false, max=<0 0>, min=<0 0>}
      }
     }
     vScrollBarPadding=<ltrb 0 0 0 0>
    }
   }
   slot {class=OverlaySlot, properties {horizontalAlignment=HAlign_Fill, verticalAlignment=VAlign_Bottom}}
  }
  BorderName {
   widgetClassPath=&Border
   bIsVariable=true
   parent=AssetTile
   widget=BorderName
   properties {brushColor=<rgba 0.0176 0.0176 0.0176 1>}
   slot {class=OverlaySlot, properties {horizontalAlignment=HAlign_Fill, verticalAlignment=VAlign_Bottom, padding=<ltrb 0 6 0 0>}}
  }
  AssetName {
   widgetClassPath=&CommonTextBlock
   bIsVariable=true
   parent=BorderName
   widget=AssetName
   properties {
    autoWrapText=true
    bIsScrollingEnabled=false
    minDesiredWidth=32
    text=Bob
    visibility=Visible
    wrapTextAt=256
    wrappingPolicy=AllowPerCharacterWrapping
    font {
     bForceMonospaced=false
     bMaterialIsStencil=false
     fontMaterial=null
     fontObject=&Roboto
     letterSpacing=0
     monospacedWidth=1
     size=13.5
     skewAmount=0
     typefaceFontName=Regular
     outlineSettings {
      bApplyOutlineToDropShadows=false
      bMiteredCorners=false
      bSeparateFillAlpha=false
      outlineMaterial=null
      outlineSize=1
      outlineColor=<rgba 0 0 0 0.5>
     }
    }
    margin=<ltrb 4 0 0 0>
    shadowOffset=<0.01 0.01>
   }
   slot {class=BorderSlot}
  }
  ContextMenuAnchor {
   widgetClassPath=/Script/UMG.MenuAnchor
   bIsVariable=true
   parent=AssetTile
   widget=ContextMenuAnchor
   properties {menuClass=&EUW_ContextMenu_C, placement=MenuPlacement_Center}
   slot {class=OverlaySlot, properties {horizontalAlignment=HAlign_Fill, verticalAlignment=VAlign_Bottom}}
  }
  DirtyFlag {
   widgetClassPath=&CommonTextBlock
   bIsVariable=true
   parent=AssetTile
   widget=DirtyFlag
   properties {
    applyLineHeightToBottomLine=false
    justification=Center
    text=*
    visibility=Collapsed
    colorAndOpacity {colorUseRule=UseColor_Specified, specifiedColor=<rgba 1 0.5 0 1>}
    font {
     bForceMonospaced=true
     bMaterialIsStencil=false
     fontMaterial=null
     fontObject=&Roboto
     letterSpacing=0
     monospacedWidth=0.5
     size=18
     skewAmount=0
     typefaceFontName=Regular
     outlineSettings {
      bApplyOutlineToDropShadows=false
      bMiteredCorners=false
      bSeparateFillAlpha=false
      outlineMaterial=null
      outlineSize=1
      outlineColor=<rgba 0 0 0 1>
     }
    }
    shadowOffset=<0 0>
   }
   slot {class=OverlaySlot, properties {padding=<ltrb 8 0 8 0>}}
  }
  Image_106 {
   widgetClassPath=&Image
   bIsVariable=true
   parent=AssetTile
   widget=Image_106
   properties {brush {
     bIsDynamicallyLoaded=false
     drawAs=Image
     imageType=FullColor
     mirroring=NoMirror
     resourceName=None
     resourceObject=/AnimationLibrary/resources/icons/locksolid
     tiling=NoTile
     imageSize=<24 24>
     margin=<ltrb 0 0 0 0>
     outlineSettings {
      bUseBrushTransparency=false
      roundingType=HalfHeightRadius
      width=0
      color {colorUseRule=UseColor_Specified, specifiedColor=<rgba 0 0 0 0>}
      cornerRadii=<0 0 0 0>
     }
     tintColor {colorUseRule=UseColor_Specified, specifiedColor=<rgba 0.8228 0.063 0.0437 1>}
     uVRegion {bIsValid=false, max=<0 0>, min=<0 0>}
    }}
   slot {class=OverlaySlot, properties {horizontalAlignment=HAlign_Right, padding=<ltrb 0 2 0 0>}}
  }
 ]
}
